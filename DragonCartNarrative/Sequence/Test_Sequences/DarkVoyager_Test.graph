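# DarkVoyager_Test is a test Level Sequence in DragonCartNarrative that animates a FortPlayerMannequin.
# The mannequin has a spawn track, a fixed world transform and a skeletal animation section.
# A Fortnite_Biped_CtrlRig track adds densely keyed body controls, which look like baked animation.
# Both FK and IK leg controls carry keys, with switch channels selecting between them.
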LevelSequence /DragonCartNarrative/Sequence/Test_Sequences/DarkVoyager_Test
bindings [
 FortPlayerMannequin {
  id=#1
  possessed_class=FortPlayerMannequin
  tracks [
   {type=MovieSceneSpawnTrack, sections [{type=MovieSceneSpawnSection, channels [None {type=MovieSceneScriptingBoolChannel, default=true}]}]}
   {type=MovieSceneSkeletalAnimationTrack, sections [{type=MovieSceneSkeletalAnimationSection, end_frame=400, start_frame=0}]}
   {
    type=MovieScene3DTransformTrack
    sections [{
      type=MovieScene3DTransformSection
      channels [
       Location.X {type=MovieSceneScriptingDoubleChannel, default=-10942.4065}
       Location.Y {type=MovieSceneScriptingDoubleChannel, default=-693.8637}
       Location.Z {type=MovieSceneScriptingDoubleChannel, default=3990.0013}
       Rotation.X {type=MovieSceneScriptingDoubleChannel, default=0}
       Rotation.Y {type=MovieSceneScriptingDoubleChannel, default=0}
       Rotation.Z {type=MovieSceneScriptingDoubleChannel, default=0}
       Scale.X {type=MovieSceneScriptingDoubleChannel, default=1}
       Scale.Y {type=MovieSceneScriptingDoubleChannel, default=1}
       Scale.Z {type=MovieSceneScriptingDoubleChannel, default=1}
      ]
     }]
   }
   {
    type=MovieSceneControlRigParameterTrack
    display_name=Fortnite_Biped_CtrlRig
    sections [
     {
      type=MovieSceneControlRigParameterSection
      channels [
       leg_l_stretch_switch {type=MovieSceneScriptingBoolChannel, default=false, const [false,322]}
       leg_r_stretch_switch {type=MovieSceneScriptingBoolChannel, default=false, const [false,322]}
       arm_l_fk_local {type=MovieSceneScriptingBoolChannel, default=true, const [true,322]}
       Auto Interpolate_3 {type=MovieSceneScriptingBoolChannel, default=false, const [false,322]}
       Auto Interpolate {type=MovieSceneScriptingBoolChannel, default=false, const [false,322]}
       arm_l_stretch_switch {type=MovieSceneScriptingBoolChannel, default=false, const [false,322]}
       Keep Aligned {type=MovieSceneScriptingBoolChannel, default=false, const [false,322]}
       arm_r_fk_local {type=MovieSceneScriptingBoolChannel, default=true, const [true,322]}
       Auto Interpolate_4 {type=MovieSceneScriptingBoolChannel, default=false, const [false,322]}
       Auto Interpolate_2 {type=MovieSceneScriptingBoolChannel, default=false, const [false,322]}
       arm_r_stretch_switch {type=MovieSceneScriptingBoolChannel, default=false, const [false,322]}
       Keep Aligned_2 {type=MovieSceneScriptingBoolChannel, default=false, const [false,322]}
       head_local {type=MovieSceneScriptingBoolChannel, default=true, const [true,322]}
       Stretch {type=MovieSceneScriptingBoolChannel, default=false, const [false,322]}
       head_stretch {type=MovieSceneScriptingBoolChannel, default=false, const [false,322]}
       spine_stretch_switch {type=MovieSceneScriptingBoolChannel, default=false, const [false,322]}
       arm_l_fk_ik_switch {type=MovieSceneScriptingBoolChannel, default=false, const [false,322]}
       leg_l_fk_ik_switch {type=MovieSceneScriptingBoolChannel, default=true, const [false,1]}
       arm_r_fk_ik_switch {type=MovieSceneScriptingBoolChannel, default=false, const [false,322]}
       leg_r_fk_ik_switch {type=MovieSceneScriptingBoolChannel, default=false}
       ShowBodyControls {type=MovieSceneScriptingBoolChannel, default=true, const [true,322]}
       spine_fk_ik_switch {type=MovieSceneScriptingBoolChannel, default=false, const [false,322]}
       neck_fk_ik_switch {type=MovieSceneScriptingBoolChannel, default=false, const [false,322]}
       space {type=MovieSceneScriptingByteChannel, default=1, const [1,322]}
       global_ctrl.Location.X {type=MovieSceneScriptingFloatChannel, default=0, const [0,322]}
       global_ctrl.Location.Y {type=MovieSceneScriptingFloatChannel, default=0, const [0,322]}
       global_ctrl.Location.Z {type=MovieSceneScriptingFloatChannel, default=0, const [0,322]}
       global_ctrl.Rotation.X {type=MovieSceneScriptingFloatChannel, default=0, const [0,322]}
       global_ctrl.Rotation.Y {type=MovieSceneScriptingFloatChannel, default=0, const [0,322]}
       global_ctrl.Rotation.Z {type=MovieSceneScriptingFloatChannel, default=0, const [0,322]}
       root_ctrl.Location.X {type=MovieSceneScriptingFloatChannel, default=0, const [0,322]}
       root_ctrl.Location.Y {type=MovieSceneScriptingFloatChannel, default=0, const [0,322]}
       root_ctrl.Location.Z {type=MovieSceneScriptingFloatChannel, default=0, const [0,322]}
       root_ctrl.Rotation.X {type=MovieSceneScriptingFloatChannel, default=0, const [0,322]}
       root_ctrl.Rotation.Y {type=MovieSceneScriptingFloatChannel, default=0, const [0,322]}
       root_ctrl.Rotation.Z {type=MovieSceneScriptingFloatChannel, default=0, const [0,322]}
       root_ctrl.Scale.X {type=MovieSceneScriptingFloatChannel, default=1, const [1,322]}
       root_ctrl.Scale.Y {type=MovieSceneScriptingFloatChannel, default=1, const [1,322]}
       root_ctrl.Scale.Z {type=MovieSceneScriptingFloatChannel, default=1, const [1,322]}
       body_offset_ctrl.Location.X {type=MovieSceneScriptingFloatChannel, default=0, const [0,322]}
       body_offset_ctrl.Location.Y {type=MovieSceneScriptingFloatChannel, default=0, const [0,322]}
       body_offset_ctrl.Location.Z {type=MovieSceneScriptingFloatChannel, default=0, const [0,322]}
       body_offset_ctrl.Rotation.X {type=MovieSceneScriptingFloatChannel, default=0, const [0,322]}
       body_offset_ctrl.Rotation.Y {type=MovieSceneScriptingFloatChannel, default=0, const [0,322]}
       body_offset_ctrl.Rotation.Z {type=MovieSceneScriptingFloatChannel, default=0, const [0,322]}
       body_offset_ctrl.Scale.X {type=MovieSceneScriptingFloatChannel, default=1, const [1,322]}
       body_offset_ctrl.Scale.Y {type=MovieSceneScriptingFloatChannel, default=1, const [1,322]}
       body_offset_ctrl.Scale.Z {type=MovieSceneScriptingFloatChannel, default=1, const [1,322]}
       thigh_l_fk_ctrl.Location.X {type=MovieSceneScriptingFloatChannel, default=0, thinned_from=320, keys [[0,0],[354,0]]}
       thigh_l_fk_ctrl.Location.Y {type=MovieSceneScriptingFloatChannel, default=0, thinned_from=320, keys [[0,0],[354,0]]}
       thigh_l_fk_ctrl.Location.Z {type=MovieSceneScriptingFloatChannel, default=0, thinned_from=320, keys [[0,0],[354,0]]}
       thigh_l_fk_ctrl.Rotation.X {
        type=MovieSceneScriptingFloatChannel
        default=1.216
        thinned_from=166
        keys [[0,18.193],[24,23.0186],[33,23.4096],[57,23.2814],[78,22.7429],[145,21.8135],[159,21.0322],[192,20.1128],[205,19.3002],[212,18.5631],[282,10.9863],[286,10.4441],[289,10.1053],[291,9.914],[295,9.6211],[298,9.4716],[301,9.3806],[304,9.3571],[312,9.5264],[320,9.8254],[354,18.193]]
       }
       thigh_l_fk_ctrl.Rotation.Y {
        type=MovieSceneScriptingFloatChannel
        default=-1.2128
        thinned_from=164
        keys [[0,0.9944],[13,3.1818],[14,3.3609],[15,3.5497],[19,4.3498],[20,4.5393],[21,4.7379],[22,4.9024],[23,5.0564],[24,5.1806],[25,5.2794],[26,5.3478],[28,5.4366],[30,5.4955],[33,5.534],[37,5.5122],[44,5.3703],[59,4.9754],[71,4.6158],[72,4.5909],[75,4.3718],[78,4.1798],[82,3.9754],[100,2.2813],[119,-2.2213],[120,-2.3744],[128,-3.278],[129,-3.287],[130,-3.2805],[132,-3.2409],[136,-3.1207],[140,-2.9726],[142,-2.8949],[145,-2.7507],[151,-2.4739],[155,-2.3114],[158,-2.216],[159,-2.2114],[160,-2.2194],[203,-1.8305],[238,-0.1601],[267,1.1059],[268,1.1226],[269,1.1107],[270,1.0831],[272,1.0121],[273,0.9749],[274,0.9417],[276,0.8914],[277,0.8739],[278,0.8686],[279,0.8675],[280,0.8712],[281,0.8787],[282,0.8798],[283,0.8836],[284,0.8848],[285,0.8839],[286,0.8851],[288,0.8766],[289,0.8697],[291,0.849],[293,0.8214],[294,0.8099],[295,0.8007],[297,0.7868],[298,0.7823],[301,0.7751],[303,0.7752],[305,0.782],[306,0.7876],[308,0.8054],[312,0.8509],[313,0.8612],[315,0.8853],[317,0.9052],[319,0.9214],[320,0.9274],[354,0.9944]]
       }
       thigh_l_fk_ctrl.Rotation.Z {
        type=MovieSceneScriptingFloatChannel
        default=-0.1664
        thinned_from=283
        keys [[0,3.0466],[21,4.6384],[22,4.5327],[25,4.1571],[27,3.8872],[29,3.6027],[33,3.0209],[34,2.8895],[38,2.4039],[41,2.0628],[43,1.8495],[45,1.649],[47,1.4602],[48,1.3727],[49,1.2889],[50,1.2121],[51,1.1478],[52,1.0953],[53,1.0499],[54,1.014],[55,0.9872],[56,0.9725],[57,0.9654],[58,0.9698],[59,0.9828],[60,1.0015],[61,1.0289],[62,1.0622],[63,1.1017],[65,1.1978],[67,1.3078],[71,1.5468],[73,1.6622],[75,1.765],[78,1.8961],[80,1.9672],[82,2.0274],[83,2.047],[84,2.0814],[85,2.1511],[86,2.2523],[87,2.3768],[88,2.5193],[89,2.6735],[91,2.9939],[92,3.1443],[93,3.2814],[94,3.3941],[95,3.4792],[96,3.5248],[97,3.5223],[98,3.476],[99,3.3903],[100,3.2706],[101,3.1218],[102,2.9386],[103,2.7175],[104,2.4498],[124,-0.2063],[125,-0.3347],[126,-0.4691],[128,-0.7484],[129,-0.8834],[130,-1.0107],[131,-1.1298],[132,-1.2386],[133,-1.3378],[134,-1.4244],[135,-1.4976],[136,-1.5548],[137,-1.5942],[138,-1.6151],[139,-1.6146],[140,-1.6003],[141,-1.581],[142,-1.5494],[143,-1.5287],[146,-1.4322],[150,-1.2944],[152,-1.2331],[153,-1.2069],[154,-1.1857],[155,-1.1697],[156,-1.1574],[157,-1.152],[158,-1.1517],[159,-1.1585],[160,-1.173],[161,-1.1463],[162,-1.1344],[163,-1.1376],[164,-1.1558],[165,-1.1844],[166,-1.2209],[167,-1.2668],[168,-1.3251],[169,-1.3977],[170,-1.4847],[171,-1.5066],[172,-1.5347],[173,-1.5681],[174,-1.6068],[175,-1.6509],[176,-1.6994],[177,-1.7517],[178,-1.8085],[180,-1.932],[183,-2.1383],[189,-2.5668],[191,-2.7007],[193,-2.8205],[196,-2.9781],[199,-3.1151],[202,-3.2319],[210,-3.5312],[212,-3.5881],[213,-3.6093],[214,-3.6205],[215,-3.6203],[216,-3.6068],[219,-3.5144],[220,-3.4768],[224,-3.3689],[228,-3.2302],[233,-3.0164],[241,-2.6325],[244,-2.505],[247,-2.3969],[250,-2.309],[255,-2.1952],[259,-2.1222],[263,-2.0711],[267,-2.0398],[277,-2.0066],[282,-1.967],[287,-1.9029],[290,-1.8522],[292,-1.812],[294,-1.7848],[295,-1.7656],[296,-1.742],[298,-1.6813],[299,-1.6446],[301,-1.5599],[302,-1.5112],[303,-1.4593],[304,-1.4024],[305,-1.342],[307,-1.2129],[310,-1.0034],[313,-0.7887],[314,-0.7194],[315,-0.6516],[316,-0.5862],[317,-0.5223],[319,-0.4017],[320,-0.3442],[354,3.0466]]
       }
       thigh_l_fk_ctrl.Scale.X {type=MovieSceneScriptingFloatChannel, default=1, const [1,320]}
       thigh_l_fk_ctrl.Scale.Y {type=MovieSceneScriptingFloatChannel, default=1, const [1,320]}
       thigh_l_fk_ctrl.Scale.Z {type=MovieSceneScriptingFloatChannel, default=1, const [1,320]}
       calf_l_fk_ctrl.Location.X {type=MovieSceneScriptingFloatChannel, default=0, thinned_from=321, keys [[0,0.013],[354,0.013]]}
       calf_l_fk_ctrl.Location.Y {type=MovieSceneScriptingFloatChannel, default=0, thinned_from=321, keys [[0,-0.0001],[354,-0.0001]]}
       calf_l_fk_ctrl.Location.Z {type=MovieSceneScriptingFloatChannel, default=0, thinned_from=321, keys [[0,0],[354,0]]}
       calf_l_fk_ctrl.Rotation.X {
        type=MovieSceneScriptingFloatChannel
        default=0.0543
        thinned_from=321
        keys [[0,0.053],[1,0.0659],[2,0.0627],[3,0.0587],[4,0.0726],[5,0.0808],[6,0.0881],[7,0.094],[8,0.0989],[10,0.1077],[12,0.1159],[19,0.1423],[20,0.1459],[21,0.1615],[26,0.1777],[28,0.1833],[30,0.1878],[33,0.1928],[36,0.1966],[45,0.2028],[49,0.2043],[50,0.204],[51,0.2046],[53,0.2042],[54,0.2046],[58,0.2041],[59,0.2033],[60,0.2033],[70,0.1958],[76,0.1893],[78,0.1865],[81,0.181],[87,0.1689],[88,0.1662],[89,0.163],[90,0.1591],[91,0.1545],[92,0.1491],[93,0.1428],[94,0.1356],[95,0.1273],[96,0.1178],[97,0.1069],[98,0.0945],[99,0.0805],[100,0.0643],[101,0.0456],[102,0.0244],[103,-0.0091],[104,-0.0091],[105,-0.0518],[106,-0.0807],[107,-0.111],[108,-0.1428],[109,-0.1788],[110,-0.2474],[112,-0.2505],[122,-0.2703],[123,-0.2237],[124,-0.203],[125,-0.1874],[126,-0.1749],[127,-0.1656],[128,-0.1588],[130,-0.1484],[132,-0.1388],[134,-0.1306],[136,-0.1229],[138,-0.1159],[140,-0.1095],[142,-0.1034],[144,-0.098],[147,-0.0907],[150,-0.0842],[154,-0.0767],[157,-0.0718],[158,-0.0706],[159,-0.0689],[160,-0.0669],[162,-0.0637],[164,-0.0613],[165,-0.0606],[166,-0.0595],[168,-0.0581],[169,-0.0583],[170,-0.0577],[171,-0.058],[175,-0.0581],[176,-0.0576],[177,-0.058],[178,-0.0576],[179,-0.0577],[180,-0.0574],[181,-0.0575],[183,-0.0568],[184,-0.0572],[186,-0.0566],[187,-0.0568],[199,-0.0537],[200,-0.0528],[201,-0.0526],[202,-0.0518],[204,-0.0518],[205,-0.052],[207,-0.054],[208,-0.0553],[212,-0.0618],[213,-0.0619],[214,-0.0588],[215,-0.0527],[216,-0.044],[217,-0.0178],[219,-0.0185],[220,0.011],[221,0.0281],[222,0.0466],[223,0.066],[224,0.0864],[225,0.1074],[226,0.1289],[227,0.1508],[233,0.2855],[234,0.3073],[235,0.3285],[236,0.3488],[237,0.3683],[238,0.3865],[239,0.4034],[240,0.4186],[241,0.4321],[242,0.4438],[243,0.4534],[244,0.461],[245,0.4661],[247,0.4728],[249,0.4769],[253,0.4807],[269,0.4704],[279,0.4587],[292,0.4455],[320,0.3985],[354,0.053]]
       }
       calf_l_fk_ctrl.Rotation.Y {
        type=MovieSceneScriptingFloatChannel
        default=1.2238
        thinned_from=321
        keys [[0,-0.0092],[1,-0.0081],[3,-0.0098],[4,-0.0103],[70,-0.0141],[89,-0.013],[102,-0.0098],[104,-0.0096],[106,-0.0099],[108,-0.0107],[109,-0.0113],[110,-0.0125],[122,-0.0161],[123,-0.0157],[161,-0.0169],[205,-0.0148],[232,-0.0186],[240,-0.0181],[251,-0.0158],[314,-0.0148],[320,-0.0142],[354,-0.0092]]
       }
       calf_l_fk_ctrl.Rotation.Z {
        type=MovieSceneScriptingFloatChannel
        default=0.3732
        thinned_from=272
        keys [[0,8.1902],[19,9.0568],[20,9.0863],[21,9.1573],[51,9.8381],[76,9.8549],[84,9.7412],[88,9.611],[92,9.3936],[103,9.2067],[153,9.8805],[158,9.8817],[202,9.2081],[206,9.2197],[213,9.3662],[218,9.569],[228,10.1521],[231,10.2645],[234,10.3038],[237,10.2637],[240,10.1505],[243,9.9781],[247,9.687],[250,9.5402],[255,9.4329],[268,9.3322],[291,9.469],[307,9.2564],[320,8.9327],[354,8.1902]]
       }
       calf_l_fk_ctrl.Scale.X {type=MovieSceneScriptingFloatChannel, default=1, const [1,321]}
       calf_l_fk_ctrl.Scale.Y {type=MovieSceneScriptingFloatChannel, default=1, const [1,321]}
       calf_l_fk_ctrl.Scale.Z {type=MovieSceneScriptingFloatChannel, default=1, const [1,321]}
       foot_l_fk_ctrl.Location.X {type=MovieSceneScriptingFloatChannel, default=0, thinned_from=319, keys [[0,0.0542],[354,0.0542]]}
       foot_l_fk_ctrl.Location.Y {type=MovieSceneScriptingFloatChannel, default=0.0398, thinned_from=319, keys [[0,0.0001],[354,0.0001]]}
       foot_l_fk_ctrl.Location.Z {type=MovieSceneScriptingFloatChannel, default=-0.8885, thinned_from=319, keys [[0,0],[354,0]]}
       foot_l_fk_ctrl.Rotation.X {
        type=MovieSceneScriptingFloatChannel
        default=-1.1647
        thinned_from=313
        keys [[0,-9.4502],[7,-9.9916],[11,-9.8517],[14,-9.6091],[15,-9.4975],[17,-9.2108],[19,-8.8534],[25,-7.6634],[27,-7.318],[28,-7.1718],[29,-7.0416],[30,-6.9426],[31,-6.8657],[32,-6.8218],[33,-6.8125],[34,-6.8306],[35,-6.8884],[36,-6.9793],[37,-7.0986],[38,-7.25],[39,-7.4236],[40,-7.6197],[41,-7.8333],[43,-8.304],[50,-10.0884],[51,-10.3248],[53,-10.7337],[54,-10.9181],[55,-11.0754],[57,-11.3141],[59,-11.4996],[61,-11.6374],[64,-11.7663],[67,-11.8165],[72,-11.7499],[82,-11.4184],[87,-11.1367],[88,-11.2128],[89,-11.5433],[90,-12.0754],[91,-12.7659],[92,-13.5841],[93,-14.477],[95,-16.3292],[96,-17.2061],[97,-17.9937],[98,-18.6478],[99,-19.1334],[100,-19.4133],[101,-19.4602],[102,-19.3555],[104,-19.0653],[110,-17.7549],[115,-16.4702],[122,-14.442],[124,-13.8112],[128,-12.6605],[129,-12.4004],[130,-12.1658],[132,-11.7706],[133,-11.6183],[134,-11.4965],[135,-11.4053],[136,-11.3535],[154,-11.3105],[158,-11.2344],[164,-10.9303],[175,-10.2079],[179,-10.0332],[185,-9.9303],[203,-10.0635],[209,-10.2783],[212,-10.4364],[213,-10.5071],[215,-10.7139],[219,-11.2432],[222,-11.7181],[226,-12.2582],[227,-12.4509],[228,-12.7446],[229,-13.1255],[230,-13.5755],[231,-14.0853],[235,-16.3021],[236,-16.814],[237,-17.2814],[238,-17.6905],[239,-18.0363],[240,-18.3123],[241,-18.5047],[242,-18.6143],[243,-18.6258],[244,-18.5289],[245,-18.3294],[246,-18.0303],[247,-17.6121],[248,-17.0583],[249,-16.3956],[250,-15.6374],[251,-14.799],[252,-13.9061],[256,-10.1261],[257,-9.2285],[258,-8.4066],[259,-7.6749],[260,-7.0345],[261,-6.5032],[262,-6.0939],[263,-5.8204],[264,-5.6827],[265,-5.7044],[266,-5.8377],[267,-6.0356],[268,-6.2861],[269,-6.5759],[270,-6.9016],[271,-7.2502],[275,-8.7103],[276,-9.039],[278,-9.6196],[279,-9.8836],[280,-10.1184],[281,-10.3159],[286,-11.1383],[289,-11.5471],[292,-11.8689],[294,-12.0198],[296,-12.1137],[298,-12.136],[302,-12.0173],[308,-11.6635],[320,-10.6818],[354,-9.4502]]
       }
       foot_l_fk_ctrl.Rotation.Y {
        type=MovieSceneScriptingFloatChannel
        default=0.0792
        thinned_from=314
        keys [[0,10.125],[6,10.0181],[12,10.3331],[15,10.5434],[18,10.8462],[20,11.0933],[21,11.2527],[27,11.998],[29,12.1922],[31,12.328],[32,12.3763],[34,12.4044],[36,12.3594],[38,12.2434],[40,12.0716],[44,11.6054],[51,10.6846],[54,10.3612],[56,10.2001],[58,10.0726],[61,9.9271],[66,9.779],[73,9.7273],[84,9.7735],[87,9.8207],[90,9.7753],[96,9.5899],[245,9.7736],[252,11.3008],[254,11.6739],[256,11.9962],[257,12.1212],[258,12.1985],[259,12.2387],[260,12.2462],[261,12.2278],[263,12.1083],[265,11.9183],[273,10.9102],[275,10.7123],[285,10.2326],[291,10.0675],[296,10.0258],[301,10.0962],[305,10.2427],[320,11.1315],[354,10.125]]
       }
       foot_l_fk_ctrl.Rotation.Z {
        type=MovieSceneScriptingFloatChannel
        default=-0.2074
        thinned_from=223
        keys [[0,29.0139],[17,30.0755],[59,30.4212],[86,29.3567],[96,28.3962],[110,27.2113],[146,22.4112],[161,20.8301],[164,20.3711],[174,18.3148],[176,17.9988],[178,17.7671],[180,17.6391],[182,17.6254],[185,17.7315],[189,18.0352],[200,19.4732],[205,20.1788],[210,20.6901],[212,20.9872],[218,22.2353],[228,24.6489],[237,27.1785],[241,28.0645],[247,28.9778],[271,30.4813],[294,30.3491],[320,28.0762],[354,29.0139]]
       }
       foot_l_fk_ctrl.Scale.X {type=MovieSceneScriptingFloatChannel, default=1, const [1,319]}
       foot_l_fk_ctrl.Scale.Y {type=MovieSceneScriptingFloatChannel, default=1, const [1,319]}
       foot_l_fk_ctrl.Scale.Z {type=MovieSceneScriptingFloatChannel, default=1, const [1,319]}
       ball_l_fk_ctrl.Location.X {type=MovieSceneScriptingFloatChannel, default=0, thinned_from=322, keys [[0,0],[354,0]]}
       ball_l_fk_ctrl.Location.Y {type=MovieSceneScriptingFloatChannel, default=0, thinned_from=322, keys [[0,0],[354,0]]}
       ball_l_fk_ctrl.Location.Z {type=MovieSceneScriptingFloatChannel, default=0, thinned_from=322, keys [[0,0],[354,0]]}
       ball_l_fk_ctrl.Rotation.X {
        type=MovieSceneScriptingFloatChannel
        default=0
        thinned_from=322
        keys [[0,0.0919],[1,0.0997],[3,0.0997],[4,0.0989],[8,0.0989],[9,0.0998],[14,0.0998],[15,0.1088],[18,0.1089],[19,0.1185],[22,0.1185],[23,0.129],[26,0.129],[27,0.1395],[31,0.1395],[32,0.146],[33,0.1465],[38,0.1465],[39,0.1424],[43,0.1424],[44,0.1324],[47,0.1324],[48,0.1204],[49,0.1209],[51,0.1209],[52,0.109],[55,0.109],[56,0.0995],[64,0.0995],[65,0.1007],[66,0.0916],[87,0.0918],[88,0.0894],[89,0.0894],[90,0.0769],[91,0.0709],[92,0.0609],[93,0.0484],[94,0.0376],[95,0.026],[96,0.0104],[97,0.0118],[98,-0.0101],[99,-0.0101],[100,-0.0265],[105,-0.0265],[106,-0.0195],[108,-0.0195],[109,-0.0074],[110,0.0011],[111,-0.827],[112,-0.8497],[113,-0.8612],[114,-0.8665],[116,-0.8665],[117,-0.8494],[118,-0.8342],[119,-0.8145],[120,-0.791],[121,-0.7632],[122,0.0615],[123,0.0641],[125,0.0641],[126,0.076],[128,0.076],[129,0.0857],[132,0.0857],[133,0.0954],[136,0.0954],[137,0.0992],[144,0.0993],[145,0.0983],[163,0.0984],[164,0.098],[166,0.098],[167,0.0983],[169,0.0983],[170,0.099],[172,0.099],[173,0.0987],[176,0.0987],[195,0.0986],[196,0.1004],[207,0.1003],[208,0.1006],[211,0.1005],[212,0.1007],[219,0.1006],[220,0.1003],[228,0.1002],[229,0.0961],[230,0.0961],[231,0.0819],[232,0.0819],[233,0.0657],[234,0.0657],[235,0.0495],[236,0.0495],[237,0.0313],[238,0.0313],[239,0.0131],[240,0.0131],[241,-0.0018],[242,-0.0018],[243,-0.0122],[249,-0.0122],[250,-0.0017],[251,-0.0017],[252,0.0144],[253,0.0268],[254,0.04],[255,0.04],[256,0.0667],[257,0.0797],[258,0.0797],[259,0.1011],[260,0.1011],[261,0.1208],[262,0.1208],[263,0.1323],[268,0.1323],[269,0.1308],[271,0.1308],[272,0.1181],[273,0.1193],[274,0.1193],[275,0.1018],[277,0.1018],[278,0.0931],[281,0.0931],[282,0.0791],[287,0.0791],[288,0.0617],[289,0.0647],[295,0.0647],[296,0.0514],[304,0.0514],[305,0.0519],[306,0.0516],[312,0.0516],[313,0.0531],[320,0.0531],[354,0.0919]]
       }
       ball_l_fk_ctrl.Rotation.Y {
        type=MovieSceneScriptingFloatChannel
        default=0
        thinned_from=322
        keys [[0,-0.1053],[1,-0.108],[3,-0.108],[4,-0.1019],[8,-0.1019],[9,-0.0975],[14,-0.0975],[15,-0.085],[16,-0.085],[17,-0.0872],[18,-0.0872],[19,-0.0728],[22,-0.0728],[23,-0.0591],[26,-0.0591],[27,-0.0455],[31,-0.0455],[32,-0.0342],[33,-0.0346],[38,-0.0346],[39,-0.0386],[43,-0.0386],[44,-0.0521],[47,-0.0521],[48,-0.0677],[49,-0.0661],[51,-0.0661],[52,-0.0831],[55,-0.0831],[56,-0.0952],[64,-0.0952],[65,-0.0922],[66,-0.1064],[80,-0.1063],[81,-0.1061],[87,-0.106],[88,-0.11],[89,-0.11],[90,-0.1303],[91,-0.144],[92,-0.1602],[93,-0.1805],[94,-0.1959],[95,-0.2105],[96,-0.2276],[97,-0.2264],[98,-0.242],[99,-0.242],[100,-0.2462],[105,-0.2462],[106,-0.2331],[108,-0.2331],[109,-0.221],[110,-0.2187],[111,0.4645],[112,0.4536],[113,0.448],[114,0.4471],[116,0.4471],[117,0.4661],[118,0.4787],[119,0.4942],[120,0.5123],[121,0.5316],[122,-0.1528],[123,-0.1473],[125,-0.1473],[126,-0.1326],[128,-0.1326],[129,-0.1199],[132,-0.1199],[133,-0.1071],[136,-0.1071],[137,-0.1041],[144,-0.104],[145,-0.1038],[163,-0.1037],[164,-0.1032],[166,-0.1032],[167,-0.103],[172,-0.1031],[173,-0.1026],[176,-0.1026],[177,-0.1014],[192,-0.1014],[207,-0.1014],[208,-0.101],[211,-0.1011],[212,-0.1009],[219,-0.101],[220,-0.1013],[224,-0.1013],[225,-0.1002],[228,-0.1002],[229,-0.1043],[230,-0.1043],[231,-0.1062],[232,-0.1062],[233,-0.1141],[234,-0.1141],[235,-0.1221],[236,-0.1221],[237,-0.1279],[238,-0.1279],[239,-0.1337],[240,-0.1337],[241,-0.1401],[242,-0.1401],[243,-0.1445],[249,-0.1445],[250,-0.14],[251,-0.14],[253,-0.1275],[254,-0.1222],[255,-0.1222],[256,-0.1116],[257,-0.1091],[258,-0.1091],[259,-0.1004],[260,-0.1004],[261,-0.095],[262,-0.095],[263,-0.0915],[268,-0.0915],[269,-0.0898],[271,-0.0898],[272,-0.0945],[274,-0.0946],[275,-0.098],[277,-0.098],[278,-0.0997],[281,-0.0997],[282,-0.1066],[287,-0.1066],[288,-0.11],[289,-0.1088],[295,-0.1088],[296,-0.1116],[304,-0.1116],[305,-0.1111],[306,-0.1115],[312,-0.1115],[313,-0.11],[320,-0.11],[354,-0.1053]]
       }
       ball_l_fk_ctrl.Rotation.Z {
        type=MovieSceneScriptingFloatChannel
        default=0
        thinned_from=322
        keys [[0,0.6363],[1,0.6667],[3,0.6667],[4,0.6769],[8,0.6769],[9,0.687],[14,0.687],[15,0.6973],[18,0.697],[19,0.7009],[22,0.7009],[23,0.6972],[26,0.6972],[27,0.7012],[31,0.7012],[32,0.7107],[38,0.7112],[39,0.7244],[43,0.7244],[44,0.7342],[47,0.7342],[48,0.744],[51,0.743],[52,0.7469],[55,0.7469],[56,0.7489],[65,0.7489],[66,0.7447],[74,0.7447],[75,0.7285],[80,0.7285],[83,0.7299],[84,0.7139],[87,0.7139],[88,0.6978],[89,0.6978],[90,0.688],[93,0.6685],[94,0.6588],[96,0.6458],[97,0.6444],[98,0.631],[99,0.631],[100,0.6243],[105,0.6243],[106,0.6177],[108,0.6177],[110,0.6577],[111,-0.973],[112,-1.0058],[113,-1.0322],[114,-1.0474],[116,-1.0474],[117,-1.0406],[118,-1.0221],[119,-1.0002],[120,-0.9698],[121,-0.9344],[122,0.6788],[123,0.6617],[125,0.6617],[126,0.6499],[128,0.6499],[129,0.639],[132,0.639],[133,0.6278],[136,0.6278],[137,0.6127],[140,0.6127],[141,0.6013],[144,0.6014],[145,0.5862],[148,0.5862],[149,0.5749],[153,0.5749],[154,0.556],[158,0.556],[159,0.5447],[160,0.5447],[161,0.5428],[163,0.5428],[164,0.53],[166,0.53],[167,0.5173],[169,0.5173],[170,0.5045],[172,0.5045],[173,0.4918],[176,0.4918],[177,0.4787],[192,0.4787],[193,0.4815],[195,0.4815],[196,0.4896],[202,0.4896],[203,0.5057],[207,0.5057],[208,0.5219],[209,0.5159],[211,0.5159],[212,0.5321],[215,0.5321],[216,0.5483],[219,0.5483],[220,0.5563],[224,0.5563],[225,0.5735],[228,0.5735],[229,0.5895],[230,0.5895],[231,0.6055],[232,0.6055],[233,0.6295],[234,0.6295],[235,0.6455],[236,0.6455],[237,0.6614],[238,0.6614],[239,0.6774],[240,0.6774],[241,0.6829],[242,0.6829],[243,0.6856],[249,0.6856],[250,0.6733],[251,0.6733],[252,0.6651],[254,0.6598],[255,0.6598],[256,0.6531],[257,0.6495],[258,0.6495],[259,0.6416],[260,0.6416],[261,0.6337],[262,0.6337],[263,0.6306],[268,0.6306],[269,0.6146],[271,0.6146],[272,0.6097],[274,0.6101],[275,0.604],[277,0.604],[278,0.5978],[281,0.5978],[282,0.5917],[287,0.5917],[288,0.5855],[295,0.5845],[296,0.5783],[304,0.5783],[305,0.5768],[306,0.5604],[312,0.5604],[313,0.5481],[320,0.5481],[354,0.6363]]
       }
       ball_l_fk_ctrl.Scale.X {type=MovieSceneScriptingFloatChannel, default=1, const [1,322]}
       ball_l_fk_ctrl.Scale.Y {type=MovieSceneScriptingFloatChannel, default=1, const [1,322]}
       ball_l_fk_ctrl.Scale.Z {type=MovieSceneScriptingFloatChannel, default=1, const [1,322]}
       foot_l_ik_ctrl.Location.X {
        type=MovieSceneScriptingFloatChannel
        default=0
        thinned_from=322
        keys [[0,-3.5597],[2,-3.5608],[3,-3.6391],[4,-3.8239],[5,-4.0251],[6,-4.1617],[9,-4.3095],[12,-4.4734],[22,-4.9076],[23,-4.9294],[35,-5.3749],[42,-5.5449],[48,-5.6385],[53,-5.6699],[60,-5.6335],[70,-5.454],[85,-5.1031],[87,-5.0282],[90,-4.8945],[93,-4.7889],[101,-4.5572],[103,-4.5278],[104,-4.5356],[105,-4.5549],[106,-4.596],[107,-4.6582],[108,-4.7506],[109,-4.8709],[110,-5.0263],[111,-5.2123],[112,-5.4688],[113,-5.6608],[120,-7.1536],[122,-7.5465],[124,-7.8882],[127,-8.1388],[131,-8.3824],[151,-9.3153],[157,-9.4934],[160,-9.5403],[164,-9.4772],[182,-8.9645],[205,-8.5613],[208,-8.5852],[214,-8.8041],[216,-8.9091],[224,-9.5027],[226,-9.6052],[228,-9.6481],[229,-9.6431],[230,-9.6121],[231,-9.5594],[232,-9.4758],[233,-9.3639],[234,-9.2213],[235,-9.0457],[236,-8.8395],[237,-8.6001],[238,-8.3289],[239,-8.0303],[240,-7.7056],[241,-7.3575],[242,-6.9891],[243,-6.6044],[245,-5.7967],[248,-4.5277],[249,-4.1199],[250,-3.7241],[251,-3.3491],[252,-2.9918],[253,-2.6557],[254,-2.3476],[255,-2.0585],[256,-1.7997],[257,-1.5589],[258,-1.3459],[259,-1.1692],[260,-1.0359],[261,-0.9528],[262,-0.9166],[263,-0.9147],[264,-0.9561],[265,-1.0313],[266,-1.1266],[267,-1.246],[268,-1.3785],[269,-1.53],[270,-1.6886],[274,-2.3543],[275,-2.5092],[276,-2.6471],[277,-2.7774],[278,-2.8779],[279,-2.962],[282,-3.1579],[285,-3.3234],[288,-3.4595],[290,-3.5298],[293,-3.6092],[295,-3.6424],[297,-3.6582],[299,-3.6548],[301,-3.63],[303,-3.5884],[304,-3.5606],[306,-3.4794],[308,-3.3745],[310,-3.2504],[312,-3.1117],[320,-2.5031],[354,-3.5597]]
       }
       foot_l_ik_ctrl.Location.Y {
        type=MovieSceneScriptingFloatChannel
        default=0
        thinned_from=322
        keys [[0,-13.3034],[1,-13.5653],[2,-14.1626],[3,-14.8022],[4,-15.1944],[6,-15.4195],[15,-16.8098],[18,-17.1673],[20,-17.3481],[21,-17.3528],[26,-17.6733],[29,-17.773],[35,-17.6799],[40,-17.4064],[46,-16.8585],[69,-13.7761],[74,-13.2054],[81,-12.5963],[83,-12.446],[96,-11.0596],[100,-10.5438],[101,-10.4298],[102,-10.3455],[103,-10.2858],[104,-10.2538],[105,-10.2737],[106,-10.3546],[107,-10.502],[108,-10.7401],[109,-11.0664],[110,-11.598],[111,-11.8914],[112,-12.2405],[113,-12.6284],[114,-13.0479],[116,-13.9761],[119,-15.4198],[120,-15.8599],[121,-16.2534],[122,-16.5979],[123,-16.6737],[124,-16.8182],[126,-17.0044],[133,-17.3764],[152,-19.2353],[157,-19.4866],[158,-19.5035],[161,-19.3164],[202,-15.1384],[204,-14.9989],[206,-14.9161],[208,-14.8874],[211,-14.9426],[212,-14.9852],[213,-15.0711],[214,-15.2252],[215,-15.4515],[216,-15.7463],[217,-16.1055],[218,-16.5368],[219,-17.0402],[220,-17.6086],[221,-18.2386],[223,-19.6389],[225,-21.1736],[230,-25.1835],[232,-26.6492],[233,-27.3031],[234,-27.8927],[235,-28.3986],[236,-28.823],[237,-29.1526],[238,-29.3764],[239,-29.485],[240,-29.4678],[241,-29.3169],[242,-29.0338],[243,-28.6037],[244,-28.0261],[245,-27.3263],[247,-25.7166],[250,-23.0833],[255,-18.58],[258,-15.9766],[262,-12.4239],[263,-11.5559],[264,-10.7244],[265,-9.915],[266,-9.1362],[267,-8.3767],[268,-7.6451],[269,-6.9543],[270,-6.2846],[271,-5.6349],[272,-5.0173],[273,-4.4174],[274,-3.8438],[275,-3.2983],[276,-2.7732],[277,-2.2823],[278,-1.8147],[279,-1.3681],[280,-0.9524],[281,-0.5578],[282,-0.1856],[283,0.1651],[284,0.4987],[285,0.808],[286,1.0995],[287,1.3678],[288,1.619],[289,1.8445],[290,2.0468],[291,2.2303],[292,2.3844],[296,2.9572],[299,3.3535],[301,3.6054],[304,3.9583],[307,4.2757],[310,4.5607],[314,4.8858],[320,5.2932],[354,-13.3034]]
       }
       foot_l_ik_ctrl.Location.Z {
        type=MovieSceneScriptingFloatChannel
        default=0
        thinned_from=322
        keys [[0,39.5321],[1,39.6152],[3,39.9776],[4,40.0588],[8,39.6332],[12,38.819],[24,35.6078],[28,34.8732],[31,34.5632],[35,34.5662],[63,37.4566],[68,37.6878],[72,37.5507],[77,36.9742],[92,34.3325],[96,33.8706],[99,33.734],[101,33.7996],[103,34.0216],[105,34.4135],[107,34.9832],[109,35.7479],[110,36.301],[114,37.9637],[122,41.7032],[130,44.0001],[135,44.94],[140,45.3615],[145,45.191],[182,39.8071],[198,38.6417],[204,38.5799],[211,39.1769],[216,39.9654],[221,41.1863],[233,45.2675],[236,45.9715],[238,46.2304],[240,46.2677],[242,46.0519],[244,45.5667],[252,43.1325],[262,41.2151],[268,40.6178],[296,40.5121],[306,39.7482],[320,37.6707],[354,39.5321]]
       }
       foot_l_ik_ctrl.Rotation.X {
        type=MovieSceneScriptingFloatChannel
        default=0
        thinned_from=322
        keys [[0,41.1122],[1,41.6685],[2,42.9829],[3,44.4977],[4,45.6831],[8,47.7192],[13,49.4871],[39,55.136],[50,56.4365],[58,56.4014],[80,53.0086],[84,51.9746],[86,51.2266],[88,50.2422],[90,48.9785],[92,47.4323],[94,45.5965],[96,43.4881],[101,37.4954],[103,35.2874],[104,34.2535],[105,33.3302],[106,32.5351],[107,31.9111],[108,31.4857],[109,31.2551],[110,31.1964],[111,31.3968],[112,31.7488],[113,32.2055],[115,33.3186],[120,36.5569],[123,38.2417],[124,38.7023],[126,39.3696],[128,39.7872],[132,40.1898],[138,40.23],[153,39.3146],[158,38.6686],[161,37.9438],[165,36.585],[173,33.4254],[176,32.4257],[179,31.6671],[182,31.1882],[186,30.8512],[192,30.7716],[202,31.3401],[206,31.8332],[209,32.3782],[212,33.2117],[213,33.6118],[214,34.1611],[215,34.8406],[216,35.6372],[217,36.5393],[218,37.5407],[220,39.8162],[225,46.2089],[227,49.1674],[230,53.9805],[233,59.0154],[235,62.0956],[236,63.4515],[237,64.6394],[238,65.6481],[239,66.4688],[240,67.097],[241,67.5114],[242,67.7094],[243,67.6791],[244,67.4246],[246,66.3767],[270,49.9579],[275,46.9738],[279,44.9701],[285,42.646],[291,40.9099],[320,34.8834],[354,41.1122]]
       }
       foot_l_ik_ctrl.Rotation.Y {
        type=MovieSceneScriptingFloatChannel
        default=0
        thinned_from=322
        keys [[0,-8.0968],[1,-8.0295],[2,-7.8696],[3,-7.6781],[4,-7.5252],[6,-7.3622],[7,-7.2403],[9,-6.9384],[11,-6.6021],[14,-6.0496],[15,-5.844],[16,-5.6151],[17,-5.3704],[18,-5.1071],[21,-4.2734],[22,-3.9871],[24,-3.4436],[25,-3.1864],[26,-2.9483],[27,-2.7306],[28,-2.5432],[29,-2.3764],[30,-2.2555],[31,-2.1658],[32,-2.1154],[33,-2.1136],[34,-2.1468],[35,-2.2257],[36,-2.345],[37,-2.5004],[38,-2.6957],[39,-2.9192],[40,-3.1732],[41,-3.4525],[42,-3.7536],[43,-4.0684],[45,-4.7394],[50,-6.4806],[51,-6.8126],[53,-7.4213],[54,-7.7035],[55,-7.9593],[57,-8.3841],[59,-8.7537],[61,-9.0755],[63,-9.351],[65,-9.5743],[67,-9.7597],[69,-9.9045],[72,-10.0493],[79,-10.1803],[83,-10.1724],[86,-10.0902],[87,-10.0392],[88,-10.1605],[89,-10.6045],[90,-11.3076],[91,-12.209],[92,-13.2561],[94,-15.5184],[95,-16.6119],[96,-17.5901],[97,-18.3904],[98,-18.9405],[99,-19.266],[100,-19.4187],[101,-19.3871],[103,-19.0591],[107,-18.1781],[110,-17.426],[113,-16.556],[117,-15.2393],[126,-12.0264],[128,-11.3912],[129,-11.1023],[130,-10.8383],[131,-10.5998],[132,-10.3895],[133,-10.2188],[134,-10.0772],[135,-9.9729],[136,-9.9125],[137,-9.8914],[177,-9.9552],[204,-10.7577],[212,-11.2318],[214,-11.3113],[215,-11.3214],[216,-11.3014],[217,-11.2525],[218,-11.1743],[219,-11.0737],[221,-10.8033],[223,-10.4484],[225,-10.0231],[226,-9.7926],[227,-9.5991],[228,-9.4824],[229,-9.4426],[230,-9.4729],[231,-9.5723],[232,-9.7294],[233,-9.9403],[234,-10.1932],[235,-10.4812],[240,-12.0925],[241,-12.3757],[242,-12.6238],[243,-12.8135],[244,-12.9255],[245,-12.9557],[246,-12.8848],[247,-12.6821],[248,-12.3322],[249,-11.8646],[250,-11.289],[251,-10.6237],[252,-9.8927],[253,-9.119],[256,-6.7127],[257,-5.9769],[258,-5.3367],[259,-4.8069],[260,-4.3923],[261,-4.1069],[262,-3.9566],[263,-3.951],[264,-4.0909],[265,-4.3859],[266,-4.7892],[267,-5.2566],[268,-5.7709],[269,-6.32],[272,-8.0484],[273,-8.6128],[274,-9.1483],[275,-9.6505],[276,-10.0986],[277,-10.4894],[278,-10.8329],[279,-11.1462],[280,-11.4237],[281,-11.6596],[284,-12.2616],[287,-12.7698],[290,-13.1842],[293,-13.501],[295,-13.6395],[297,-13.7079],[299,-13.7044],[302,-13.5993],[305,-13.3948],[309,-12.9856],[320,-11.5721],[354,-8.0968]]
       }
       foot_l_ik_ctrl.Rotation.Z {
        type=MovieSceneScriptingFloatChannel
        default=0
        thinned_from=322
        keys [[0,-28.7847],[1,-29.2632],[2,-30.3828],[3,-31.6626],[4,-32.6245],[7,-33.7969],[9,-34.4267],[11,-34.8695],[13,-35.1417],[28,-35.4572],[78,-34.2629],[85,-34.6273],[87,-34.8332],[92,-35.8479],[93,-35.9651],[94,-35.9963],[95,-35.9177],[96,-35.7117],[97,-35.3762],[99,-34.3868],[100,-33.9874],[101,-33.7491],[104,-33.456],[106,-33.4777],[107,-33.5724],[108,-33.754],[109,-34.034],[110,-34.4068],[111,-34.8966],[114,-36.6461],[115,-37.1274],[116,-37.5334],[125,-40.1965],[131,-41.463],[136,-42.0372],[209,-42.9507],[219,-42.7814],[221,-42.5825],[223,-42.1171],[225,-41.3857],[226,-40.9445],[227,-40.398],[228,-39.7421],[229,-38.9778],[230,-38.1241],[232,-36.1871],[235,-34.4773],[238,-32.5337],[241,-30.3821],[247,-25.6529],[249,-24.1919],[250,-23.5258],[251,-22.9134],[252,-22.3547],[253,-21.8543],[255,-21.0164],[257,-20.3041],[261,-19.1009],[264,-18.2435],[275,-14.3394],[277,-13.7106],[279,-13.1532],[281,-12.6835],[284,-12.0999],[287,-11.5976],[290,-11.1878],[292,-10.9714],[293,-10.8855],[295,-10.7754],[297,-10.7449],[299,-10.789],[301,-10.8847],[305,-11.1865],[320,-12.5415],[354,-28.7847]]
       }
       foot_l_ik_ctrl.Scale.X {type=MovieSceneScriptingFloatChannel, default=1, const [1,322]}
       foot_l_ik_ctrl.Scale.Y {type=MovieSceneScriptingFloatChannel, default=1, const [1,322]}
       foot_l_ik_ctrl.Scale.Z {type=MovieSceneScriptingFloatChannel, default=1, const [1,322]}
       ball_l_ik_ctrl.Location.X {type=MovieSceneScriptingFloatChannel, default=0, thinned_from=322, keys [[0,0],[354,0]]}
       ball_l_ik_ctrl.Location.Y {type=MovieSceneScriptingFloatChannel, default=0, thinned_from=322, keys [[0,0],[354,0]]}
       ball_l_ik_ctrl.Location.Z {type=MovieSceneScriptingFloatChannel, default=0, thinned_from=322, keys [[0,0],[354,0]]}
       ball_l_ik_ctrl.Rotation.X {
        type=MovieSceneScriptingFloatChannel
        default=0
        thinned_from=322
        keys [[0,0.6252],[1,0.6553],[3,0.6553],[4,0.666],[8,0.666],[9,0.6765],[14,0.6765],[15,0.6877],[18,0.6873],[19,0.6924],[22,0.6924],[23,0.6899],[26,0.6899],[27,0.695],[31,0.695],[32,0.7055],[38,0.7059],[39,0.7187],[43,0.7187],[44,0.7274],[47,0.7273],[48,0.7358],[51,0.735],[52,0.7374],[65,0.7386],[66,0.7332],[74,0.7332],[75,0.717],[80,0.717],[83,0.7184],[84,0.7024],[87,0.7024],[88,0.6861],[89,0.6861],[90,0.6746],[93,0.6508],[94,0.6397],[96,0.6239],[97,0.6226],[98,0.6078],[99,0.6078],[100,0.6007],[105,0.6007],[106,0.5953],[108,0.5953],[110,0.6365],[111,-0.9222],[112,-0.9558],[113,-0.9825],[114,-0.9976],[116,-0.9976],[117,-0.9891],[118,-0.9695],[119,-0.9463],[120,-0.9146],[121,-0.8777],[122,0.6634],[123,0.6469],[125,0.6469],[126,0.6364],[128,0.6364],[129,0.6267],[132,0.6267],[133,0.6166],[136,0.6166],[137,0.6018],[140,0.6018],[141,0.5905],[144,0.5905],[145,0.5755],[148,0.5755],[149,0.5642],[153,0.5642],[154,0.5454],[158,0.5454],[159,0.5341],[160,0.5341],[161,0.5322],[163,0.5322],[164,0.5195],[166,0.5195],[167,0.5068],[169,0.5068],[170,0.4941],[172,0.4941],[173,0.4815],[176,0.4815],[177,0.4685],[192,0.4685],[193,0.4713],[195,0.4713],[196,0.4794],[202,0.4794],[203,0.4955],[207,0.4955],[208,0.5117],[209,0.5056],[211,0.5056],[212,0.5218],[215,0.5218],[216,0.5379],[219,0.5379],[220,0.5459],[224,0.5459],[225,0.5631],[228,0.5631],[229,0.5787],[230,0.5787],[231,0.5944],[232,0.5944],[233,0.6176],[234,0.6176],[235,0.6328],[236,0.6328],[237,0.6481],[238,0.6481],[239,0.6635],[240,0.6635],[241,0.6683],[242,0.6683],[243,0.6706],[249,0.6706],[250,0.6587],[251,0.6587],[252,0.6512],[254,0.647],[255,0.647],[256,0.6413],[257,0.638],[258,0.638],[259,0.631],[260,0.631],[261,0.6236],[262,0.6236],[263,0.6208],[268,0.6208],[269,0.605],[271,0.605],[272,0.5997],[274,0.6001],[275,0.5936],[277,0.5936],[278,0.5874],[281,0.5874],[282,0.5806],[287,0.5806],[288,0.5741],[295,0.5732],[296,0.5668],[304,0.5668],[305,0.5653],[306,0.5489],[312,0.5489],[313,0.5368],[320,0.5368],[354,0.6252]]
       }
       ball_l_ik_ctrl.Rotation.Y {
        type=MovieSceneScriptingFloatChannel
        default=0
        thinned_from=322
        keys [[0,-0.1578],[1,-0.1629],[3,-0.1629],[4,-0.1577],[8,-0.1577],[9,-0.1541],[14,-0.1541],[15,-0.1424],[16,-0.1424],[17,-0.1445],[18,-0.1445],[19,-0.1303],[22,-0.1303],[23,-0.1162],[26,-0.1162],[27,-0.1028],[31,-0.1028],[32,-0.0921],[33,-0.0925],[38,-0.0925],[39,-0.0977],[43,-0.0977],[44,-0.1123],[47,-0.1123],[48,-0.1288],[49,-0.1272],[51,-0.1272],[52,-0.1446],[55,-0.1447],[56,-0.157],[64,-0.157],[65,-0.1541],[66,-0.1681],[74,-0.1681],[75,-0.1666],[83,-0.1665],[84,-0.1651],[87,-0.1651],[88,-0.1677],[89,-0.1677],[90,-0.1873],[91,-0.2006],[92,-0.2164],[93,-0.2363],[95,-0.2652],[96,-0.282],[97,-0.2806],[98,-0.2955],[99,-0.2955],[100,-0.2994],[105,-0.2994],[106,-0.2857],[108,-0.2857],[109,-0.2751],[110,-0.2743],[111,0.5529],[112,0.5456],[113,0.5428],[116,0.5434],[117,0.5615],[118,0.572],[119,0.5851],[121,0.6153],[122,-0.2093],[123,-0.2023],[125,-0.2023],[126,-0.1864],[128,-0.1864],[129,-0.1727],[132,-0.1727],[133,-0.1588],[136,-0.1588],[137,-0.1544],[140,-0.1544],[141,-0.1534],[144,-0.1534],[145,-0.152],[148,-0.152],[149,-0.1509],[153,-0.1509],[154,-0.1494],[158,-0.1494],[159,-0.1484],[163,-0.1482],[164,-0.1467],[166,-0.1467],[167,-0.1454],[169,-0.1454],[170,-0.1443],[172,-0.1443],[173,-0.1428],[176,-0.1428],[177,-0.1405],[195,-0.1406],[196,-0.1413],[202,-0.1413],[203,-0.1427],[207,-0.1427],[208,-0.1437],[209,-0.1433],[211,-0.1433],[212,-0.1445],[215,-0.1445],[216,-0.1459],[219,-0.1459],[220,-0.147],[224,-0.147],[225,-0.1473],[228,-0.1473],[229,-0.1528],[230,-0.1528],[231,-0.1563],[232,-0.1563],[233,-0.1666],[234,-0.1666],[235,-0.1761],[236,-0.1761],[237,-0.1836],[238,-0.1836],[239,-0.1912],[240,-0.1912],[241,-0.1983],[242,-0.1983],[243,-0.2031],[249,-0.2031],[250,-0.1974],[251,-0.1974],[252,-0.1902],[253,-0.1834],[254,-0.1777],[255,-0.1777],[256,-0.166],[257,-0.1629],[258,-0.1629],[259,-0.1532],[260,-0.1532],[261,-0.1468],[262,-0.1468],[263,-0.1428],[268,-0.1428],[269,-0.1398],[271,-0.1398],[272,-0.1444],[274,-0.1445],[275,-0.1477],[277,-0.1477],[278,-0.149],[281,-0.149],[282,-0.1556],[287,-0.1556],[288,-0.1588],[289,-0.1574],[295,-0.1574],[296,-0.1599],[304,-0.1599],[305,-0.1593],[306,-0.1583],[312,-0.1583],[313,-0.1557],[320,-0.1557],[354,-0.1578]]
       }
       ball_l_ik_ctrl.Rotation.Z {
        type=MovieSceneScriptingFloatChannel
        default=0
        thinned_from=322
        keys [[0,-0.0942],[1,-0.1021],[3,-0.1021],[4,-0.1012],[8,-0.1012],[9,-0.102],[14,-0.102],[15,-0.1109],[18,-0.1109],[19,-0.1203],[22,-0.1203],[23,-0.1305],[26,-0.1305],[27,-0.1407],[31,-0.1407],[32,-0.1471],[33,-0.1475],[38,-0.1475],[39,-0.1436],[43,-0.1436],[44,-0.1338],[47,-0.1338],[48,-0.1222],[49,-0.1227],[51,-0.1227],[52,-0.1112],[55,-0.1112],[56,-0.1019],[64,-0.1019],[65,-0.103],[66,-0.0942],[87,-0.0943],[88,-0.092],[89,-0.092],[90,-0.0798],[91,-0.0741],[92,-0.0644],[93,-0.0523],[94,-0.0417],[95,-0.0304],[96,-0.015],[97,-0.0165],[98,0.0052],[99,0.0052],[100,0.0216],[105,0.0216],[106,0.0149],[108,0.0149],[109,0.0028],[110,-0.0057],[111,0.8225],[112,0.8449],[113,0.8562],[114,0.8614],[116,0.8614],[117,0.8442],[118,0.829],[119,0.8095],[120,0.7861],[121,0.7585],[122,-0.0649],[123,-0.0673],[125,-0.0673],[126,-0.0788],[128,-0.0788],[129,-0.0882],[132,-0.0882],[133,-0.0977],[136,-0.0977],[137,-0.1014],[144,-0.1014],[145,-0.1004],[163,-0.1004],[164,-0.1],[166,-0.1],[169,-0.1002],[170,-0.1009],[172,-0.1009],[173,-0.1005],[176,-0.1005],[192,-0.1003],[195,-0.1004],[196,-0.1022],[207,-0.1021],[208,-0.1025],[211,-0.1024],[212,-0.1026],[219,-0.1026],[220,-0.1023],[228,-0.1022],[229,-0.0982],[230,-0.0982],[231,-0.0841],[232,-0.0841],[233,-0.0681],[234,-0.0681],[235,-0.0522],[236,-0.0522],[237,-0.0341],[238,-0.0341],[239,-0.0161],[240,-0.0161],[241,-0.0013],[242,-0.0013],[243,0.009],[249,0.009],[250,-0.0014],[251,-0.0014],[252,-0.0174],[253,-0.0296],[254,-0.0427],[255,-0.0427],[256,-0.0691],[257,-0.0821],[258,-0.0821],[259,-0.1033],[260,-0.1033],[261,-0.1228],[262,-0.1228],[263,-0.1343],[268,-0.1343],[269,-0.1327],[271,-0.1327],[272,-0.1201],[273,-0.1213],[274,-0.1213],[275,-0.1039],[277,-0.1039],[278,-0.0952],[281,-0.0952],[282,-0.0813],[287,-0.0813],[288,-0.0639],[289,-0.0669],[295,-0.0669],[296,-0.0536],[304,-0.0536],[305,-0.0541],[306,-0.0538],[312,-0.0538],[313,-0.0552],[320,-0.0552],[354,-0.0942]]
       }
       ball_l_ik_ctrl.Scale.X {type=MovieSceneScriptingFloatChannel, default=1, const [1,322]}
       ball_l_ik_ctrl.Scale.Y {type=MovieSceneScriptingFloatChannel, default=1, const [1,322]}
       ball_l_ik_ctrl.Scale.Z {type=MovieSceneScriptingFloatChannel, default=1, const [1,322]}
       foot_bk1_l_ctrl.Location.X {type=MovieSceneScriptingFloatChannel, default=0, const [0,322]}
       foot_bk1_l_ctrl.Location.Y {type=MovieSceneScriptingFloatChannel, default=0, const [0,322]}
       foot_bk1_l_ctrl.Location.Z {type=MovieSceneScriptingFloatChannel, default=0, const [0,322]}
       foot_bk1_l_ctrl.Rotation.X {type=MovieSceneScriptingFloatChannel, default=0, const [0,322]}
       foot_bk1_l_ctrl.Rotation.Y {type=MovieSceneScriptingFloatChannel, default=0, const [0,322]}
       foot_bk1_l_ctrl.Rotation.Z {type=MovieSceneScriptingFloatChannel, default=0, const [0,322]}
       foot_bk1_l_ctrl.Scale.X {type=MovieSceneScriptingFloatChannel, default=1, const [1,322]}
       foot_bk1_l_ctrl.Scale.Y {type=MovieSceneScriptingFloatChannel, default=1, const [1,322]}
       foot_bk1_l_ctrl.Scale.Z {type=MovieSceneScriptingFloatChannel, default=1, const [1,322]}
       foot_roll_l_ctrl.Location.X {type=MovieSceneScriptingFloatChannel, default=0, const [0,322]}
       foot_roll_l_ctrl.Location.Y {type=MovieSceneScriptingFloatChannel, default=0, const [0,322]}
       foot_roll_l_ctrl.Location.Z {type=MovieSceneScriptingFloatChannel, default=0, const [0,322]}
       foot_roll_l_ctrl.Rotation.X {type=MovieSceneScriptingFloatChannel, default=0, const [0,322]}
       foot_roll_l_ctrl.Rotation.Y {type=MovieSceneScriptingFloatChannel, default=0, const [0,322]}
       foot_roll_l_ctrl.Rotation.Z {type=MovieSceneScriptingFloatChannel, default=0, const [0,322]}
       foot_roll_l_ctrl.Scale.X {type=MovieSceneScriptingFloatChannel, default=1, const [1,322]}
       foot_roll_l_ctrl.Scale.Y {type=MovieSceneScriptingFloatChannel, default=1, const [1,322]}
       foot_roll_l_ctrl.Scale.Z {type=MovieSceneScriptingFloatChannel, default=1, const [1,322]}
       roll_blend_l_ctrl {type=MovieSceneScriptingFloatChannel, default=1, const [1,322]}
       tip_l_ctrl.Location.X {type=MovieSceneScriptingFloatChannel, default=0, const [0,322]}
       tip_l_ctrl.Location.Y {type=MovieSceneScriptingFloatChannel, default=0, const [0,322]}
       tip_l_ctrl.Location.Z {type=MovieSceneScriptingFloatChannel, default=0, const [0,322]}
       tip_l_ctrl.Rotation.X {type=MovieSceneScriptingFloatChannel, default=0, const [0,322]}
       tip_l_ctrl.Rotation.Y {type=MovieSceneScriptingFloatChannel, default=0, const [0,322]}
       tip_l_ctrl.Rotation.Z {type=MovieSceneScriptingFloatChannel, default=0, const [0,322]}
       tip_l_ctrl.Scale.X {type=MovieSceneScriptingFloatChannel, default=1, const [1,322]}
       tip_l_ctrl.Scale.Y {type=MovieSceneScriptingFloatChannel, default=1, const [1,322]}
       tip_l_ctrl.Scale.Z {type=MovieSceneScriptingFloatChannel, default=1, const [1,322]}
       heel_l_ctrl.Location.X {type=MovieSceneScriptingFloatChannel, default=0, const [0,322]}
       heel_l_ctrl.Location.Y {type=MovieSceneScriptingFloatChannel, default=0, const [0,322]}
       heel_l_ctrl.Location.Z {type=MovieSceneScriptingFloatChannel, default=0, const [0,322]}
       heel_l_ctrl.Rotation.X {type=MovieSceneScriptingFloatChannel, default=0, const [0,322]}
       heel_l_ctrl.Rotation.Y {type=MovieSceneScriptingFloatChannel, default=0, const [0,322]}
       heel_l_ctrl.Rotation.Z {type=MovieSceneScriptingFloatChannel, default=0, const [0,322]}
       heel_l_ctrl.Scale.X {type=MovieSceneScriptingFloatChannel, default=1, const [1,322]}
       heel_l_ctrl.Scale.Y {type=MovieSceneScriptingFloatChannel, default=1, const [1,322]}
       heel_l_ctrl.Scale.Z {type=MovieSceneScriptingFloatChannel, default=1, const [1,322]}
       leg_l_softness {type=MovieSceneScriptingFloatChannel, default=0, const [0,322]}
       leg_l_pv_ik_ctrl.Location.X {
        type=MovieSceneScriptingFloatChannel
        default=0
        thinned_from=322
        keys [[0,-0.3844],[1,-0.503],[2,-0.7626],[3,-0.8366],[4,-0.9326],[6,-1.0068],[7,-1.0613],[8,-1.1239],[9,-0.9065],[10,-1.2458],[14,-1.4839],[15,-1.5473],[17,-1.6862],[19,-1.8418],[20,-1.9238],[21,-2.0154],[23,-2.1742],[26,-2.3902],[28,-2.5104],[29,-2.5617],[30,-2.6059],[31,-2.6434],[33,-2.6917],[34,-2.7057],[35,-2.7085],[36,-2.705],[37,-2.6959],[39,-2.6594],[42,-2.5825],[45,-2.4871],[48,-2.3845],[51,-2.2902],[54,-2.2083],[57,-2.144],[58,-2.1272],[59,-2.1029],[66,-1.9761],[67,-1.9556],[79,-1.7493],[82,-1.6856],[83,-1.6599],[84,-1.6306],[85,-1.5945],[86,-1.5506],[87,-1.4997],[88,-1.4411],[89,-1.3535],[90,-1.2292],[91,-1.0594],[92,-0.8303],[93,-0.5351],[94,-0.1716],[95,0.2609],[96,0.7669],[97,1.3348],[98,1.9589],[99,2.5573],[100,3.0331],[101,3.3543],[102,3.5699],[103,3.7627],[104,3.9259],[105,4.0515],[106,4.1413],[107,4.1887],[108,4.1903],[109,4.1501],[110,4.0955],[111,3.9337],[112,3.7434],[113,3.5272],[114,3.301],[115,3.0561],[117,2.5387],[120,1.7377],[121,1.4826],[122,1.2379],[123,0.9884],[124,0.7585],[125,0.5421],[126,0.3462],[127,0.1686],[128,0.0049],[129,-0.1405],[130,-0.2719],[131,-0.3897],[132,-0.4947],[133,-0.5807],[134,-0.6515],[135,-0.7058],[136,-0.7402],[137,-0.7521],[138,-0.7565],[141,-0.7599],[146,-0.773],[149,-0.7841],[151,-0.787],[154,-0.7861],[157,-0.7773],[158,-0.7723],[159,-0.7622],[161,-0.7358],[162,-0.7198],[164,-0.6823],[167,-0.616],[169,-0.5693],[171,-0.5253],[173,-0.4844],[174,-0.4658],[175,-0.4485],[177,-0.4205],[179,-0.4017],[180,-0.3938],[182,-0.3848],[183,-0.383],[185,-0.3817],[186,-0.3817],[187,-0.3853],[188,-0.386],[189,-0.3882],[192,-0.4009],[194,-0.4122],[197,-0.4322],[199,-0.4467],[202,-0.4711],[203,-0.4805],[204,-0.4919],[205,-0.5058],[207,-0.538],[208,-0.5554],[209,-0.575],[211,-0.6196],[212,-0.6441],[213,-0.6772],[214,-0.7242],[215,-0.7833],[216,-0.853],[217,-0.9322],[218,-1.0206],[219,-1.1189],[220,-1.2247],[221,-1.3371],[223,-1.5792],[225,-1.8307],[226,-1.9663],[227,-2.0651],[228,-2.0984],[229,-2.0763],[230,-2.008],[231,-1.9011],[232,-1.768],[233,-1.6124],[234,-1.4413],[235,-1.2555],[236,-1.0628],[239,-0.4609],[240,-0.2647],[241,-0.0759],[242,0.1029],[243,0.2659],[244,0.4132],[245,0.5357],[246,0.6193],[247,0.6313],[248,0.5672],[249,0.4336],[250,0.2409],[251,-0.0096],[252,-0.3026],[253,-0.6314],[254,-0.9915],[255,-1.3686],[257,-2.1443],[258,-2.5248],[259,-2.8827],[260,-3.2138],[261,-3.5027],[262,-3.7429],[263,-3.9207],[264,-4.0311],[265,-4.0583],[266,-4.0287],[267,-3.965],[268,-3.8742],[269,-3.7622],[270,-3.6272],[271,-3.4746],[272,-3.3077],[273,-3.1335],[276,-2.5788],[277,-2.4017],[278,-2.232],[279,-2.0728],[280,-1.9345],[281,-1.8209],[286,-1.3151],[287,-1.219],[288,-1.1264],[289,-1.0371],[290,-0.9502],[291,-0.8688],[292,-0.7911],[293,-0.7211],[294,-0.6559],[295,-0.596],[296,-0.5458],[297,-0.5007],[298,-0.4654],[299,-0.4396],[300,-0.4153],[302,-0.3738],[303,-0.3508],[304,-0.3304],[306,-0.2873],[308,-0.2477],[309,-0.2293],[311,-0.1934],[312,-0.177],[314,-0.1491],[315,-0.1362],[316,-0.1246],[317,-0.1124],[318,-0.1039],[319,-0.094],[320,-0.0858],[354,-0.3844]]
       }
       leg_l_pv_ik_ctrl.Location.Y {
        type=MovieSceneScriptingFloatChannel
        default=0
        thinned_from=322
        keys [[0,4.0213],[1,3.5433],[2,3.926],[3,4.3448],[4,4.5851],[6,4.6924],[8,4.8193],[9,1.763],[10,4.9317],[20,5.3733],[21,5.4783],[32,5.9353],[46,6.3113],[56,6.4644],[67,6.4654],[79,6.298],[82,6.2283],[84,6.1593],[87,6.0115],[89,5.8777],[91,5.7038],[93,5.4877],[94,5.3624],[95,5.2222],[96,5.0686],[97,4.8999],[100,4.3335],[101,4.1744],[102,4.0475],[103,3.9533],[104,3.9027],[105,3.9088],[106,3.9751],[107,4.1061],[108,4.3065],[109,4.586],[110,5.1511],[111,5.2831],[112,5.4375],[119,6.6681],[121,6.9873],[122,7.1292],[123,6.9806],[124,6.9818],[127,7.0549],[141,7.5017],[142,7.5207],[151,7.7261],[158,7.7725],[165,7.5481],[176,7.129],[199,6.6626],[203,6.6172],[204,6.6212],[206,6.6651],[212,6.9332],[214,7.06],[217,7.315],[219,7.5272],[226,8.3663],[228,8.5701],[230,8.7299],[232,8.8336],[233,8.8611],[235,8.8638],[237,8.8001],[239,8.6719],[241,8.4894],[243,8.2592],[246,7.8311],[247,7.7073],[248,7.6035],[249,7.5198],[250,7.4536],[252,7.366],[264,7.1166],[266,7.0985],[269,7.1238],[279,7.3989],[286,7.5083],[292,7.5088],[301,7.3574],[307,7.1514],[320,6.5559],[354,4.0213]]
       }
       leg_l_pv_ik_ctrl.Location.Z {
        type=MovieSceneScriptingFloatChannel
        default=0
        thinned_from=322
        keys [[0,4.2355],[1,3.6665],[2,4.0815],[3,4.569],[4,4.8483],[6,4.971],[8,5.1144],[9,1.6739],[10,5.2406],[14,5.4542],[20,5.7323],[21,5.8582],[33,6.4448],[49,6.9975],[56,7.1357],[64,7.1693],[73,7.0756],[81,6.8906],[83,6.8198],[86,6.6574],[88,6.5151],[91,6.2336],[94,5.9002],[99,5.2978],[101,5.0469],[102,4.9422],[103,4.8733],[104,4.8509],[105,4.8879],[106,4.9888],[107,5.1582],[108,5.4021],[109,5.7366],[110,6.4359],[111,6.5654],[112,6.7189],[120,8.1609],[122,8.4856],[123,8.232],[124,8.1951],[126,8.1892],[133,8.3869],[141,8.7168],[142,8.7438],[149,8.9921],[155,9.0993],[158,9.1086],[162,8.9529],[175,8.2698],[194,7.6861],[202,7.5216],[204,7.5156],[206,7.5719],[208,7.6696],[212,7.9265],[213,8.0021],[215,8.1989],[217,8.4291],[219,8.7119],[226,9.8684],[229,10.3045],[231,10.5268],[232,10.6082],[233,10.6669],[234,10.7001],[235,10.7094],[236,10.6943],[237,10.6561],[238,10.5937],[240,10.4052],[242,10.1404],[243,9.9806],[246,9.3956],[248,9.0527],[250,8.7841],[252,8.5805],[255,8.3634],[262,8.005],[264,7.9372],[266,7.912],[268,7.9297],[271,8.0295],[277,8.3518],[285,8.6396],[290,8.7185],[295,8.6992],[301,8.5559],[307,8.2803],[320,7.4797],[354,4.2355]]
       }
       leg_l_pv_ik_ctrl.Rotation.X {
        type=MovieSceneScriptingFloatChannel
        default=0
        thinned_from=322
        keys [[0,-5.7093],[1,-4.9733],[2,-5.5129],[3,-6.14],[4,-6.4969],[6,-6.6533],[8,-6.8358],[9,-2.4519],[10,-6.9962],[14,-7.2669],[20,-7.6186],[21,-7.7766],[33,-8.5097],[49,-9.1883],[56,-9.3564],[65,-9.3915],[76,-9.2109],[82,-9.0153],[84,-8.9086],[87,-8.6827],[89,-8.4829],[91,-8.2387],[94,-7.8186],[99,-7.0476],[101,-6.7259],[102,-6.592],[103,-6.5033],[104,-6.474],[105,-6.5206],[106,-6.6484],[107,-6.863],[108,-7.1712],[109,-7.5918],[110,-8.4631],[111,-8.6241],[112,-8.8144],[120,-10.5802],[122,-10.9726],[123,-10.6694],[124,-10.6261],[126,-10.6211],[133,-10.8634],[141,-11.2591],[142,-11.2914],[150,-11.6165],[157,-11.7266],[158,-11.7257],[164,-11.4182],[176,-10.6704],[199,-9.886],[203,-9.8056],[204,-9.8106],[206,-9.8791],[208,-9.998],[212,-10.3095],[214,-10.5117],[217,-10.9158],[219,-11.2552],[226,-12.629],[229,-13.1401],[231,-13.3984],[233,-13.5599],[234,-13.5977],[235,-13.6074],[237,-13.5436],[239,-13.3699],[241,-13.1023],[243,-12.7495],[246,-12.0588],[248,-11.6525],[250,-11.3344],[252,-11.094],[255,-10.8393],[262,-10.4213],[264,-10.3412],[266,-10.3108],[268,-10.3312],[271,-10.4493],[277,-10.831],[285,-11.1704],[290,-11.2622],[295,-11.2373],[301,-11.065],[307,-10.7338],[320,-9.7648],[354,-5.7093]]
       }
       leg_l_pv_ik_ctrl.Rotation.Y {
        type=MovieSceneScriptingFloatChannel
        default=0
        thinned_from=322
        keys [[0,0.0378],[1,0.0206],[2,0.014],[3,0.0197],[4,0.0203],[6,0.0189],[7,0.0174],[8,0.0156],[9,-1.2474],[10,0.0117],[11,0.0098],[14,0.0034],[15,0.0009],[17,-0.0046],[19,-0.0111],[20,-0.0146],[21,-0.0167],[23,-0.0232],[25,-0.0289],[26,-0.0316],[28,-0.0358],[29,-0.0374],[30,-0.0387],[31,-0.0396],[33,-0.0399],[34,-0.0397],[35,-0.0388],[36,-0.0376],[37,-0.0361],[38,-0.0342],[40,-0.0298],[42,-0.0249],[44,-0.0194],[45,-0.0168],[46,-0.014],[50,-0.0036],[51,-0.0012],[54,0.0049],[57,0.0095],[58,0.0107],[59,0.0122],[63,0.0165],[67,0.02],[72,0.023],[84,0.0277],[85,0.0282],[86,0.0291],[87,0.0301],[88,0.0314],[89,0.034],[90,0.0383],[91,0.045],[92,0.0547],[93,0.0678],[94,0.0842],[95,0.1039],[96,0.1271],[97,0.1529],[98,0.1808],[99,0.2067],[100,0.2263],[101,0.2387],[103,0.254],[104,0.2608],[105,0.2669],[106,0.2725],[107,0.2774],[109,0.286],[110,0.2972],[111,0.293],[112,0.288],[114,0.2765],[117,0.2554],[121,0.2249],[122,0.2176],[123,0.1991],[124,0.1871],[125,0.1764],[126,0.1669],[127,0.1587],[128,0.1513],[129,0.1447],[130,0.1389],[131,0.1339],[132,0.1295],[133,0.1262],[134,0.1237],[135,0.122],[136,0.1215],[137,0.1221],[141,0.1264],[142,0.127],[143,0.1284],[147,0.1318],[153,0.1354],[158,0.1371],[180,0.1271],[186,0.1227],[202,0.1084],[203,0.1076],[204,0.1072],[205,0.107],[206,0.1071],[207,0.1074],[212,0.1103],[213,0.1107],[214,0.1108],[215,0.1106],[216,0.1102],[220,0.1071],[221,0.1062],[223,0.1039],[225,0.101],[226,0.099],[227,0.0987],[228,0.1016],[229,0.1072],[230,0.1146],[231,0.1235],[232,0.1329],[235,0.1622],[236,0.1713],[237,0.18],[238,0.1879],[239,0.195],[240,0.2013],[241,0.2066],[242,0.2109],[243,0.2138],[244,0.2154],[245,0.2154],[246,0.2135],[247,0.2089],[248,0.2013],[249,0.1908],[250,0.178],[251,0.1627],[252,0.1455],[253,0.1267],[254,0.1063],[255,0.0849],[258,0.0184],[259,-0.0029],[260,-0.0229],[261,-0.0407],[262,-0.0558],[263,-0.0672],[264,-0.0745],[265,-0.0766],[266,-0.0751],[267,-0.0713],[268,-0.0656],[269,-0.0584],[270,-0.0497],[271,-0.0397],[272,-0.0288],[273,-0.0173],[275,0.0068],[276,0.0184],[277,0.0295],[278,0.0399],[279,0.0498],[280,0.0583],[281,0.0655],[283,0.0783],[285,0.0905],[286,0.0963],[288,0.1071],[290,0.1167],[291,0.121],[292,0.125],[293,0.1284],[294,0.1314],[295,0.1341],[296,0.1362],[297,0.138],[298,0.1392],[300,0.1401],[303,0.1403],[306,0.1395],[309,0.1377],[312,0.1354],[317,0.1303],[320,0.1269],[354,0.0378]]
       }
       leg_l_pv_ik_ctrl.Rotation.Z {
        type=MovieSceneScriptingFloatChannel
        default=0
        thinned_from=322
        keys [[0,0.5277],[1,0.6735],[2,0.9974],[3,1.0876],[4,1.2044],[6,1.2947],[7,1.3609],[8,1.4368],[9,0.0398],[10,1.5843],[14,1.871],[15,1.9471],[17,2.1136],[19,2.2998],[20,2.3977],[21,2.5052],[23,2.6938],[26,2.9486],[28,3.0893],[29,3.149],[30,3.2002],[31,3.2434],[33,3.2979],[34,3.3133],[35,3.3154],[37,3.2979],[39,3.2521],[42,3.1577],[51,2.8044],[54,2.7063],[57,2.6295],[58,2.6095],[59,2.5806],[79,2.1666],[82,2.0923],[83,2.0625],[84,2.0283],[85,1.9861],[86,1.9347],[87,1.8749],[88,1.8059],[89,1.7022],[90,1.5543],[91,1.3516],[92,1.0767],[93,0.721],[94,0.281],[95,-0.2448],[96,-0.8631],[97,-1.5607],[98,-2.3322],[99,-3.0776],[100,-3.6762],[101,-4.0859],[102,-4.3643],[103,-4.6119],[104,-4.819],[105,-4.973],[106,-5.0754],[107,-5.1177],[108,-5.0951],[109,-5.0123],[110,-4.8794],[111,-4.67],[112,-4.4252],[113,-4.1489],[114,-3.8619],[115,-3.5542],[117,-2.9118],[118,-2.5813],[120,-1.9352],[121,-1.6281],[122,-1.3355],[123,-1.0511],[124,-0.7823],[125,-0.5285],[126,-0.2987],[127,-0.0904],[128,0.1016],[129,0.272],[130,0.4258],[131,0.5637],[132,0.6865],[133,0.7869],[134,0.8696],[135,0.9329],[136,0.973],[137,0.987],[138,0.9923],[141,0.9966],[146,1.0123],[149,1.0253],[152,1.0291],[155,1.0257],[157,1.0178],[158,1.012],[159,1.0003],[161,0.9696],[162,0.9508],[164,0.907],[167,0.8295],[169,0.7746],[171,0.7228],[173,0.6745],[174,0.6524],[175,0.6319],[177,0.5986],[179,0.5761],[180,0.5665],[182,0.5554],[184,0.5517],[186,0.5508],[187,0.5547],[188,0.5554],[189,0.5577],[192,0.572],[194,0.585],[197,0.6081],[199,0.6248],[202,0.6532],[203,0.6642],[204,0.6777],[205,0.6942],[207,0.7327],[208,0.7534],[209,0.7768],[211,0.8298],[212,0.8588],[213,0.8979],[214,0.9532],[215,1.0225],[216,1.1039],[217,1.1963],[218,1.299],[219,1.4126],[220,1.5345],[221,1.6635],[223,1.9394],[225,2.2239],[226,2.3766],[227,2.4869],[228,2.5226],[229,2.4956],[230,2.4166],[231,2.2944],[232,2.1432],[233,1.9669],[234,1.7734],[235,1.5634],[236,1.3456],[239,0.663],[240,0.4392],[241,0.2229],[242,0.017],[243,-0.1721],[244,-0.3444],[245,-0.4894],[246,-0.5903],[247,-0.6085],[248,-0.5379],[249,-0.3861],[250,-0.1646],[251,0.1249],[252,0.4646],[253,0.8465],[254,1.2652],[255,1.7041],[257,2.6073],[258,3.0506],[259,3.468],[260,3.8544],[261,4.192],[262,4.473],[263,4.6815],[264,4.8112],[265,4.8439],[266,4.8097],[267,4.7354],[268,4.6292],[269,4.4977],[270,4.3393],[271,4.1601],[272,3.9642],[273,3.7598],[276,3.1107],[277,2.9041],[278,2.7063],[279,2.5209],[280,2.36],[281,2.2277],[286,1.6402],[287,1.5286],[288,1.4213],[289,1.3178],[290,1.2171],[291,1.1228],[292,1.0327],[293,0.9515],[294,0.8758],[295,0.8063],[296,0.7479],[297,0.6955],[298,0.6543],[299,0.6241],[300,0.5956],[302,0.5468],[303,0.5197],[304,0.4956],[306,0.4443],[308,0.3969],[311,0.3314],[312,0.3114],[314,0.2771],[315,0.2612],[316,0.2467],[317,0.2316],[318,0.2208],[319,0.2083],[320,0.1979],[354,0.5277]]
       }
       leg_l_pv_ik_ctrl.Scale.X {type=MovieSceneScriptingFloatChannel, default=1, const [1,322]}
       leg_l_pv_ik_ctrl.Scale.Y {type=MovieSceneScriptingFloatChannel, default=1, const [1,322]}
       leg_l_pv_ik_ctrl.Scale.Z {type=MovieSceneScriptingFloatChannel, default=1, const [1,322]}
       thigh_r_fk_ctrl.Location.X {type=MovieSceneScriptingFloatChannel, default=0, thinned_from=321, keys [[0,0],[354,0]]}
       thigh_r_fk_ctrl.Location.Y {type=MovieSceneScriptingFloatChannel, default=0, thinned_from=321, keys [[0,0],[354,0]]}
       thigh_r_fk_ctrl.Location.Z {type=MovieSceneScriptingFloatChannel, default=0, thinned_from=321, keys [[0,0],[354,0]]}
       thigh_r_fk_ctrl.Rotation.X {
        type=MovieSceneScriptingFloatChannel
        default=1.2139
        thinned_from=172
        keys [[0,-1.5643],[14,-0.9535],[24,-0.6543],[36,-0.805],[51,-1.08],[57,-1.2522],[61,-1.3609],[70,-1.5872],[94,-2.1832],[95,-2.204],[96,-2.2156],[97,-2.2173],[149,-1.245],[153,-1.248],[158,-1.239],[177,-1.2371],[180,-1.2441],[182,-1.2542],[194,-1.3399],[199,-1.3645],[203,-1.3688],[212,-1.3605],[213,-1.363],[214,-1.3621],[215,-1.3648],[217,-1.3647],[218,-1.362],[219,-1.348],[220,-1.2887],[236,-0.6194],[245,-1.1487],[251,-2.2432],[253,-3.2267],[254,-3.7005],[255,-4.1545],[256,-4.5887],[257,-4.9955],[258,-5.3743],[259,-5.7126],[260,-6.01],[261,-6.2641],[262,-6.4769],[263,-6.6531],[264,-6.7911],[265,-6.8916],[266,-6.9477],[267,-6.9559],[268,-6.9106],[269,-6.8097],[273,-6.3148],[276,-5.9762],[278,-5.7667],[280,-5.5847],[282,-5.43],[283,-5.3608],[285,-5.2543],[287,-5.181],[289,-5.1493],[315,-5.1496],[320,-5.199],[354,-1.5643]]
       }
       thigh_r_fk_ctrl.Rotation.Y {
        type=MovieSceneScriptingFloatChannel
        default=-1.213
        thinned_from=61
        keys [[0,-3.9023],[13,-5.3017],[32,-7.6618],[50,-8.0466],[71,-6.1964],[75,-5.9603],[94,-5.069],[95,-5.0029],[96,-4.9241],[97,-4.829],[98,-4.7139],[107,-3.2228],[143,-0.9971],[144,-1.0078],[200,-0.9328],[225,-4.0892],[242,-7.5086],[266,-10.1963],[282,-10.477],[287,-10.133],[320,-8.912],[354,-3.9023]]
       }
       thigh_r_fk_ctrl.Rotation.Z {
        type=MovieSceneScriptingFloatChannel
        default=-0.1592
        thinned_from=78
        keys [[0,-18.4346],[13,-18.5942],[23,-19.0805],[59,-20.9337],[64,-20.7351],[92,-18.1042],[99,-17.6406],[112,-17.3217],[148,-20.9251],[198,-22.2889],[254,-19.1751],[259,-18.8408],[263,-18.7744],[267,-18.894],[274,-19.433],[280,-20.0994],[286,-20.5019],[291,-20.6102],[320,-19.3626],[354,-18.4346]]
       }
       thigh_r_fk_ctrl.Scale.X {type=MovieSceneScriptingFloatChannel, default=1, const [1,321]}
       thigh_r_fk_ctrl.Scale.Y {type=MovieSceneScriptingFloatChannel, default=1, const [1,321]}
       thigh_r_fk_ctrl.Scale.Z {type=MovieSceneScriptingFloatChannel, default=1, const [1,321]}
       calf_r_fk_ctrl.Location.X {type=MovieSceneScriptingFloatChannel, default=0, thinned_from=321, keys [[0,0],[354,0]]}
       calf_r_fk_ctrl.Location.Y {type=MovieSceneScriptingFloatChannel, default=0, thinned_from=321, keys [[0,0],[354,0]]}
       calf_r_fk_ctrl.Location.Z {type=MovieSceneScriptingFloatChannel, default=0, thinned_from=321, keys [[0,0],[354,0]]}
       calf_r_fk_ctrl.Rotation.X {
        type=MovieSceneScriptingFloatChannel
        default=0.0561
        thinned_from=321
        keys [[0,0.9331],[1,0.9163],[2,0.9175],[12,0.8876],[15,0.8835],[20,0.8794],[24,0.885],[69,0.9061],[78,0.8996],[82,0.8893],[84,0.8785],[86,0.8614],[88,0.8391],[91,0.7989],[97,0.7128],[99,0.6877],[100,0.6767],[101,0.6671],[102,0.6589],[103,0.6523],[104,0.6474],[105,0.6442],[106,0.6428],[107,0.6433],[109,0.6491],[111,0.6593],[114,0.6804],[122,0.7479],[125,0.7683],[127,0.7773],[137,0.8008],[146,0.8087],[162,0.8056],[192,0.7343],[199,0.7268],[213,0.7346],[219,0.747],[224,0.7664],[228,0.7906],[241,0.8903],[245,0.9121],[248,0.9215],[260,0.9307],[284,0.9867],[302,0.9832],[320,0.9553],[354,0.9331]]
       }
       calf_r_fk_ctrl.Rotation.Y {
        type=MovieSceneScriptingFloatChannel
        default=1.1982
        thinned_from=321
        keys [[0,-0.0201],[45,-0.0197],[62,-0.0199],[84,-0.0193],[89,-0.0181],[94,-0.0157],[102,-0.0106],[104,-0.0097],[106,-0.0094],[109,-0.0099],[120,-0.0151],[126,-0.0168],[170,-0.0174],[201,-0.0149],[221,-0.016],[240,-0.0196],[253,-0.0202],[290,-0.0203],[301,-0.0204],[302,-0.0203],[354,-0.0201]]
       }
       calf_r_fk_ctrl.Rotation.Z {
        type=MovieSceneScriptingFloatChannel
        default=0.2616
        thinned_from=229
        keys [[0,31.6746],[14,31.2467],[46,29.9222],[77,29.1751],[82,28.7914],[84,28.5091],[86,28.0879],[96,26.5535],[103,26.0619],[107,25.9942],[151,27.3104],[162,27.1671],[172,26.338],[191,24.384],[199,24.0025],[211,24.2157],[219,24.7232],[224,25.2794],[243,28.7678],[247,29.2995],[293,31.4214],[320,31.3406],[354,31.6746]]
       }
       calf_r_fk_ctrl.Scale.X {type=MovieSceneScriptingFloatChannel, default=1, const [1,321]}
       calf_r_fk_ctrl.Scale.Y {type=MovieSceneScriptingFloatChannel, default=1, const [1,321]}
       calf_r_fk_ctrl.Scale.Z {type=MovieSceneScriptingFloatChannel, default=1, const [1,321]}
       foot_r_fk_ctrl.Location.X {type=MovieSceneScriptingFloatChannel, default=0, thinned_from=319, keys [[0,0.0542],[1,0.0301],[2,0.0181],[320,0.0181],[354,0.0542]]}
       foot_r_fk_ctrl.Location.Y {type=MovieSceneScriptingFloatChannel, default=-0.0329, thinned_from=319, keys [[0,0],[354,0]]}
       foot_r_fk_ctrl.Location.Z {type=MovieSceneScriptingFloatChannel, default=-0.9075, thinned_from=319, keys [[0,-0.0001],[1,0.0001],[354,-0.0001]]}
       foot_r_fk_ctrl.Rotation.X {
        type=MovieSceneScriptingFloatChannel
        default=-1.1648
        thinned_from=319
        keys [[0,-6.4078],[1,-6.1656],[2,-5.8398],[3,-5.6917],[4,-5.5591],[5,-5.456],[6,-5.3761],[7,-5.3078],[11,-5.159],[13,-5.1518],[14,-5.1667],[15,-5.2022],[16,-5.2705],[17,-5.3691],[18,-5.4906],[19,-5.6348],[21,-5.9897],[25,-6.7774],[27,-7.1389],[28,-7.3066],[29,-7.453],[30,-7.5799],[31,-7.6793],[32,-7.7523],[33,-7.7905],[34,-7.7978],[35,-7.7613],[36,-7.6965],[37,-7.6031],[38,-7.4754],[39,-7.3287],[40,-7.1582],[42,-6.7649],[43,-6.5522],[45,-6.0887],[49,-5.1378],[50,-4.9121],[51,-4.6968],[52,-4.4914],[53,-4.3025],[54,-4.1378],[55,-3.9862],[56,-3.8648],[57,-3.7619],[58,-3.6691],[59,-3.591],[60,-3.523],[61,-3.4631],[63,-3.3752],[64,-3.3493],[66,-3.3115],[68,-3.3081],[70,-3.3263],[72,-3.3647],[80,-3.5577],[82,-3.5939],[83,-3.6177],[84,-3.6578],[86,-3.7666],[87,-3.8024],[88,-3.7551],[89,-3.5717],[90,-3.2847],[91,-2.9112],[92,-2.4757],[93,-2.0089],[94,-1.5308],[95,-1.0699],[96,-0.6366],[97,-0.2389],[98,0.1015],[99,0.3668],[100,0.5449],[101,0.6154],[102,0.6205],[103,0.617],[104,0.5995],[105,0.5699],[106,0.5329],[107,0.4821],[108,0.4208],[109,0.3453],[110,0.2579],[111,0.1585],[112,0.0477],[113,-0.0803],[114,-0.2201],[115,-0.3712],[116,-0.536],[118,-0.8627],[119,-1.0235],[120,-1.1816],[121,-1.3342],[122,-1.4744],[123,-1.6056],[124,-1.7259],[125,-1.8369],[126,-1.9413],[127,-2.0379],[128,-2.1268],[129,-2.2081],[130,-2.2768],[131,-2.3401],[132,-2.3918],[133,-2.4337],[134,-2.4651],[135,-2.4871],[136,-2.5009],[137,-2.4997],[142,-2.4056],[152,-2.1577],[157,-2.0623],[161,-2.0206],[172,-1.9642],[177,-1.9606],[180,-1.9747],[185,-2.0286],[198,-2.1787],[202,-2.2104],[205,-2.2165],[215,-2.1804],[216,-2.1888],[217,-2.2144],[218,-2.2526],[223,-2.4882],[228,-2.7644],[230,-2.8702],[231,-2.9322],[232,-3.0036],[233,-3.0847],[234,-3.1754],[235,-3.278],[236,-3.3914],[237,-3.4926],[238,-3.5617],[239,-3.5982],[240,-3.5998],[241,-3.5799],[242,-3.532],[243,-3.4555],[244,-3.3641],[245,-3.2448],[246,-3.0946],[247,-2.9297],[248,-2.7423],[249,-2.5396],[251,-2.0902],[252,-1.8539],[254,-1.3927],[255,-1.1687],[256,-0.9579],[257,-0.754],[258,-0.5671],[259,-0.3944],[260,-0.2324],[261,-0.081],[262,0.0485],[263,0.1596],[264,0.2492],[265,0.3101],[266,0.3401],[267,0.3592],[268,0.3914],[269,0.4297],[270,0.4763],[271,0.525],[272,0.5839],[273,0.6366],[275,0.7471],[277,0.8494],[278,0.9026],[280,0.9942],[281,1.0376],[283,1.1083],[284,1.1355],[285,1.1552],[286,1.1703],[287,1.1718],[288,1.1665],[289,1.148],[290,1.1222],[292,1.0663],[295,0.9745],[296,0.9453],[299,0.8484],[300,0.8125],[301,0.779],[303,0.7068],[306,0.603],[309,0.5062],[310,0.4767],[311,0.4486],[312,0.4225],[313,0.3932],[314,0.3689],[316,0.3223],[317,0.3025],[318,0.2803],[319,0.2603],[320,0.2413],[354,-6.4078]]
       }
       foot_r_fk_ctrl.Rotation.Y {
        type=MovieSceneScriptingFloatChannel
        default=0.1067
        thinned_from=219
        keys [[0,-18.2191],[1,-18.1034],[2,-17.9135],[7,-17.3957],[14,-16.3731],[25,-14.343],[28,-13.8815],[30,-13.6427],[32,-13.4721],[34,-13.3867],[36,-13.3915],[38,-13.4723],[41,-13.7028],[54,-15.2804],[61,-15.8239],[72,-16.3237],[85,-16.6059],[88,-16.7136],[95,-16.5703],[97,-16.6013],[102,-16.9083],[105,-16.9491],[113,-16.7308],[127,-16.7184],[157,-15.6014],[178,-15.5625],[203,-16.1226],[218,-16.0535],[274,-17.912],[306,-18.2024],[354,-18.2191]]
       }
       foot_r_fk_ctrl.Rotation.Z {
        type=MovieSceneScriptingFloatChannel
        default=-0.1036
        thinned_from=202
        keys [[0,24.3976],[98,23.2221],[99,23.0086],[100,22.8468],[102,22.7098],[114,22.7672],[117,22.6277],[119,22.4475],[120,22.3126],[121,22.1034],[122,21.7797],[123,21.3528],[124,20.8432],[125,20.2758],[129,17.7226],[130,17.1092],[131,16.5525],[132,16.0582],[133,15.654],[134,15.3623],[135,15.1884],[136,15.1603],[151,15.928],[160,16.1862],[162,16.1633],[163,16.1223],[166,15.8604],[169,15.4229],[177,13.8767],[179,13.5423],[181,13.2755],[184,12.9661],[189,12.5723],[194,12.3307],[199,12.2272],[203,12.2336],[204,12.2043],[206,12.082],[208,11.9164],[232,12.8183],[233,13.2212],[235,13.9331],[237,14.5747],[240,15.4063],[242,15.8585],[244,16.2131],[247,16.6088],[256,17.32],[259,17.6582],[267,19.091],[272,20.1414],[275,20.9245],[291,25.9075],[295,26.9623],[300,27.9957],[305,28.7098],[309,29.0162],[314,29.0518],[320,28.738],[354,24.3976]]
       }
       foot_r_fk_ctrl.Scale.X {type=MovieSceneScriptingFloatChannel, default=1, const [1,319]}
       foot_r_fk_ctrl.Scale.Y {type=MovieSceneScriptingFloatChannel, default=1, const [1,319]}
       foot_r_fk_ctrl.Scale.Z {type=MovieSceneScriptingFloatChannel, default=1, const [1,319]}
       ball_r_fk_ctrl.Location.X {type=MovieSceneScriptingFloatChannel, default=0, thinned_from=322, keys [[0,0],[354,0]]}
       ball_r_fk_ctrl.Location.Y {type=MovieSceneScriptingFloatChannel, default=0, thinned_from=322, keys [[0,0],[354,0]]}
       ball_r_fk_ctrl.Location.Z {type=MovieSceneScriptingFloatChannel, default=0, thinned_from=322, keys [[0,0],[354,0]]}
       ball_r_fk_ctrl.Rotation.X {
        type=MovieSceneScriptingFloatChannel
        default=0
        thinned_from=322
        keys [[0,0.1012],[1,0.1025],[4,0.1026],[5,0.1015],[6,0.1015],[7,0.0986],[9,0.0986],[10,0.0933],[12,0.0933],[13,0.0863],[15,0.0863],[16,0.0768],[17,0.0793],[19,0.0793],[20,0.061],[22,0.061],[23,0.0472],[25,0.0472],[26,0.034],[28,0.034],[29,0.0231],[32,0.0231],[33,0.0237],[34,0.0237],[35,0.0146],[40,0.0146],[41,0.0295],[44,0.0295],[45,0.0444],[47,0.0444],[48,0.0592],[49,0.0577],[51,0.0577],[52,0.0729],[56,0.0729],[57,0.0866],[64,0.0866],[65,0.0871],[66,0.1052],[74,0.1052],[75,0.1016],[80,0.1016],[81,0.1026],[83,0.1026],[84,0.1033],[86,0.1033],[87,0.103],[88,0.103],[89,0.1062],[90,0.1103],[91,0.1168],[92,0.1238],[93,0.1301],[94,0.1301],[95,0.1451],[96,0.1451],[97,0.1591],[99,0.1591],[100,0.1722],[112,0.1722],[113,0.1597],[117,0.1597],[118,0.1477],[121,0.1477],[122,0.1352],[123,0.1352],[124,0.1293],[125,0.1293],[126,0.1235],[127,0.1235],[128,0.1188],[129,0.1192],[130,0.1138],[131,0.1138],[132,0.1096],[133,0.1096],[134,0.1064],[144,0.1064],[145,0.1068],[160,0.1068],[161,0.107],[171,0.107],[172,0.1051],[176,0.1051],[177,0.1049],[186,0.105],[187,0.1041],[192,0.1041],[208,0.1043],[209,0.1046],[210,0.1039],[217,0.1039],[219,0.103],[220,0.1034],[221,0.103],[222,0.1034],[224,0.1033],[226,0.1041],[227,0.104],[229,0.1048],[232,0.1049],[233,0.1053],[234,0.1053],[235,0.1057],[237,0.1057],[238,0.1072],[240,0.1072],[241,0.1155],[242,0.1155],[243,0.1251],[244,0.1251],[245,0.1378],[246,0.1378],[247,0.1506],[248,0.1506],[249,0.1678],[250,0.1678],[251,0.1761],[252,0.1863],[253,0.1863],[254,0.2022],[255,0.2022],[256,0.215],[257,0.2131],[258,0.2245],[259,0.2245],[260,0.2378],[262,0.2378],[263,0.2491],[268,0.2491],[269,0.2571],[272,0.2571],[273,0.2562],[276,0.2562],[277,0.2582],[279,0.2582],[280,0.2603],[282,0.2603],[283,0.2621],[285,0.2621],[286,0.2639],[287,0.2639],[288,0.2655],[289,0.2652],[290,0.2666],[292,0.2666],[293,0.2677],[295,0.2677],[296,0.2691],[298,0.2691],[299,0.2702],[303,0.2702],[304,0.2719],[319,0.2717],[320,0.2714],[354,0.1012]]
       }
       ball_r_fk_ctrl.Rotation.Y {
        type=MovieSceneScriptingFloatChannel
        default=0
        thinned_from=322
        keys [[0,-0.0922],[1,-0.091],[4,-0.0911],[5,-0.0924],[6,-0.0924],[7,-0.0895],[9,-0.0895],[10,-0.0892],[12,-0.0892],[13,-0.0846],[15,-0.0846],[16,-0.0826],[17,-0.0838],[19,-0.0838],[20,-0.0712],[22,-0.0712],[23,-0.0746],[25,-0.0746],[26,-0.0671],[28,-0.0671],[29,-0.0676],[32,-0.0676],[34,-0.0678],[35,-0.0647],[40,-0.0647],[41,-0.0696],[44,-0.0696],[45,-0.0745],[47,-0.0745],[48,-0.0794],[49,-0.0791],[51,-0.0791],[52,-0.0822],[56,-0.0822],[57,-0.0868],[64,-0.0868],[65,-0.0886],[66,-0.093],[74,-0.093],[75,-0.0893],[80,-0.0893],[81,-0.0916],[83,-0.0916],[84,-0.0902],[86,-0.0902],[87,-0.0899],[88,-0.0899],[89,-0.0869],[90,-0.0847],[91,-0.0809],[92,-0.0775],[93,-0.0756],[94,-0.0756],[95,-0.07],[96,-0.07],[97,-0.0651],[99,-0.0651],[100,-0.062],[112,-0.062],[113,-0.0675],[117,-0.0675],[118,-0.0705],[121,-0.0705],[122,-0.0765],[123,-0.0765],[124,-0.078],[125,-0.078],[126,-0.0796],[127,-0.0796],[128,-0.0823],[129,-0.0815],[130,-0.0833],[131,-0.0833],[132,-0.0839],[133,-0.0839],[134,-0.0855],[144,-0.0855],[145,-0.0847],[160,-0.0847],[171,-0.0849],[176,-0.085],[177,-0.0846],[186,-0.0847],[187,-0.0844],[192,-0.0844],[209,-0.0842],[210,-0.085],[217,-0.085],[219,-0.0856],[222,-0.0868],[223,-0.0876],[224,-0.0875],[226,-0.0884],[227,-0.0883],[229,-0.089],[230,-0.089],[231,-0.0887],[232,-0.0887],[233,-0.089],[234,-0.089],[235,-0.0894],[237,-0.0894],[238,-0.092],[240,-0.092],[241,-0.1038],[242,-0.1038],[243,-0.1173],[244,-0.1173],[245,-0.1352],[246,-0.1352],[247,-0.1531],[248,-0.1531],[249,-0.1742],[250,-0.1742],[251,-0.189],[252,-0.2018],[253,-0.2018],[254,-0.2242],[255,-0.2242],[256,-0.2421],[257,-0.2403],[258,-0.2576],[259,-0.2576],[260,-0.2729],[262,-0.2729],[263,-0.2902],[268,-0.2902],[269,-0.2964],[272,-0.2964],[273,-0.2913],[276,-0.2913],[277,-0.2901],[279,-0.2901],[280,-0.2888],[282,-0.2888],[283,-0.2874],[285,-0.2874],[286,-0.2859],[287,-0.2859],[288,-0.2852],[289,-0.2853],[290,-0.284],[292,-0.284],[293,-0.2833],[295,-0.2833],[296,-0.282],[298,-0.282],[299,-0.2813],[303,-0.2813],[304,-0.2794],[319,-0.2796],[320,-0.2806],[354,-0.0922]]
       }
       ball_r_fk_ctrl.Rotation.Z {
        type=MovieSceneScriptingFloatChannel
        default=0
        thinned_from=322
        keys [[0,0.56],[1,0.5431],[2,0.5431],[3,0.5262],[4,0.5262],[5,0.5093],[6,0.5093],[7,0.4924],[9,0.4924],[10,0.4755],[12,0.4755],[13,0.4585],[15,0.4586],[16,0.4416],[19,0.4425],[20,0.4295],[22,0.4295],[23,0.4231],[28,0.423],[29,0.4294],[32,0.4294],[33,0.427],[34,0.427],[35,0.4419],[40,0.4419],[41,0.451],[44,0.451],[45,0.463],[47,0.463],[48,0.4691],[49,0.468],[51,0.468],[52,0.4733],[56,0.4733],[57,0.4786],[64,0.4786],[65,0.4804],[66,0.4751],[74,0.4751],[75,0.4642],[80,0.4642],[81,0.4631],[83,0.4631],[84,0.4527],[86,0.4527],[87,0.4319],[88,0.4319],[89,0.406],[92,0.3437],[93,0.3282],[94,0.3282],[95,0.3075],[96,0.3075],[97,0.2967],[99,0.2967],[100,0.2873],[112,0.2873],[113,0.2881],[121,0.288],[122,0.2774],[123,0.2774],[124,0.2562],[125,0.2562],[126,0.235],[127,0.235],[128,0.2138],[129,0.2179],[130,0.1944],[131,0.1944],[132,0.1787],[133,0.1787],[134,0.163],[160,0.163],[161,0.1603],[171,0.1603],[172,0.1467],[176,0.1467],[177,0.1508],[178,0.1291],[186,0.1291],[187,0.1183],[192,0.1183],[193,0.1196],[208,0.1196],[209,0.1235],[210,0.104],[217,0.104],[224,0.2402],[225,0.2606],[228,0.3127],[229,0.3214],[230,0.3214],[231,0.3387],[232,0.3387],[233,0.3561],[234,0.3561],[235,0.3735],[237,0.3735],[238,0.3909],[240,0.3909],[241,0.3995],[242,0.3995],[243,0.4044],[244,0.4044],[245,0.4069],[246,0.4069],[247,0.4078],[250,0.408],[252,0.4066],[255,0.406],[259,0.4065],[260,0.4102],[262,0.4101],[263,0.4138],[268,0.4138],[269,0.4278],[272,0.4278],[273,0.4374],[276,0.4374],[277,0.4549],[279,0.4549],[280,0.4725],[282,0.4725],[283,0.49],[285,0.49],[286,0.5075],[287,0.5075],[288,0.5192],[289,0.521],[290,0.5344],[292,0.5344],[293,0.5479],[295,0.5479],[296,0.5614],[298,0.5614],[299,0.5703],[303,0.5703],[304,0.5838],[305,0.5856],[319,0.5857],[320,0.5699],[354,0.56]]
       }
       ball_r_fk_ctrl.Scale.X {type=MovieSceneScriptingFloatChannel, default=1, const [1,322]}
       ball_r_fk_ctrl.Scale.Y {type=MovieSceneScriptingFloatChannel, default=1, const [1,322]}
       ball_r_fk_ctrl.Scale.Z {type=MovieSceneScriptingFloatChannel, default=1, const [1,322]}
       foot_r_ik_ctrl.Location.X {
        type=MovieSceneScriptingFloatChannel
        default=0
        thinned_from=322
        keys [[0,-9.9153],[1,-9.5307],[2,-9.5283],[15,-8.1231],[18,-7.846],[20,-7.6923],[26,-7.3338],[33,-7.0398],[41,-6.8429],[47,-6.7718],[53,-6.7784],[58,-6.8552],[61,-6.9331],[81,-7.6544],[83,-7.7219],[85,-7.845],[94,-8.6368],[99,-9.1278],[101,-9.2748],[103,-9.3378],[105,-9.331],[107,-9.2634],[113,-8.9384],[115,-8.8712],[121,-8.7124],[152,-7.3712],[157,-7.2184],[162,-7.1475],[168,-7.1558],[179,-7.3355],[195,-7.7915],[202,-7.9046],[212,-7.9106],[216,-7.8498],[218,-7.7929],[220,-7.6931],[223,-7.4921],[224,-7.4441],[226,-7.3956],[228,-7.4039],[229,-7.4251],[231,-7.512],[233,-7.6439],[235,-7.8068],[237,-8.0074],[241,-8.4981],[246,-9.1879],[250,-9.6361],[255,-10.0613],[259,-10.3128],[261,-10.3814],[264,-10.3895],[268,-10.2861],[274,-9.9582],[281,-9.4974],[285,-9.2961],[288,-9.2136],[299,-9.2663],[320,-9.8186],[354,-9.9153]]
       }
       foot_r_ik_ctrl.Location.Y {
        type=MovieSceneScriptingFloatChannel
        default=0
        thinned_from=322
        keys [[0,-1.1277],[1,-0.5574],[2,-0.3893],[3,-0.1508],[4,0.086],[5,0.3202],[6,0.5568],[7,0.7862],[8,1.0035],[9,1.2067],[10,1.3916],[11,1.5588],[12,1.7092],[13,1.8287],[14,1.9356],[15,2.0222],[16,2.0844],[18,2.2767],[19,2.3679],[20,2.4515],[23,2.6631],[24,2.7249],[25,2.7431],[26,2.7541],[27,2.7587],[29,2.7338],[31,2.688],[35,2.5629],[36,2.5242],[37,2.4945],[39,2.4268],[41,2.3651],[42,2.3451],[43,2.3199],[45,2.2923],[46,2.2852],[47,2.2845],[49,2.2993],[51,2.3403],[52,2.3698],[53,2.4096],[55,2.5041],[56,2.5699],[57,2.6428],[58,2.732],[60,2.9298],[62,3.1687],[64,3.4372],[66,3.7238],[69,4.1862],[73,4.8288],[75,5.1395],[78,5.6336],[80,6.0088],[82,6.4367],[83,6.7061],[84,7.0333],[85,7.407],[86,7.8164],[87,8.2414],[91,10.0026],[93,10.8308],[94,11.2156],[95,11.5752],[96,11.9031],[97,12.1923],[98,12.4417],[99,12.6454],[100,12.7595],[101,12.7616],[102,12.6343],[103,12.3953],[104,12.0498],[105,11.6018],[106,11.0519],[107,10.4176],[108,9.711],[109,8.9683],[110,8.2059],[112,6.6441],[113,5.895],[114,5.1798],[115,4.5295],[118,2.6777],[119,2.0674],[120,1.4687],[121,0.8876],[122,0.3205],[123,-0.2121],[124,-0.7162],[125,-1.1755],[126,-1.5891],[127,-1.9577],[128,-2.2881],[129,-2.5982],[130,-2.8971],[131,-3.1682],[132,-3.4203],[133,-3.6436],[134,-3.8401],[135,-4.0123],[136,-4.1595],[137,-4.2781],[138,-4.365],[139,-4.4305],[140,-4.4684],[141,-4.4944],[143,-4.4909],[145,-4.4269],[147,-4.3271],[149,-4.1916],[155,-3.7215],[156,-3.6527],[160,-3.3956],[162,-3.2517],[164,-3.059],[165,-2.949],[166,-2.8196],[167,-2.6995],[168,-2.5695],[171,-2.1613],[173,-1.882],[175,-1.6105],[176,-1.4784],[178,-1.2248],[180,-0.9873],[181,-0.8734],[182,-0.7629],[183,-0.66],[184,-0.5604],[185,-0.4648],[186,-0.3787],[187,-0.2987],[188,-0.2278],[189,-0.166],[190,-0.1115],[191,-0.0584],[192,-0.0198],[193,0.0145],[195,0.0675],[196,0.0788],[197,0.096],[198,0.0982],[199,0.101],[200,0.0974],[202,0.0736],[203,0.047],[204,0.0104],[205,-0.0253],[206,-0.0684],[207,-0.1221],[208,-0.1825],[209,-0.2614],[210,-0.3478],[211,-0.4604],[212,-0.5858],[213,-0.7368],[214,-0.9123],[215,-1.1098],[216,-1.3337],[217,-1.5818],[218,-1.863],[219,-2.1883],[220,-2.554],[221,-2.9615],[223,-3.8379],[224,-4.2685],[225,-4.6886],[227,-5.4992],[229,-6.2731],[231,-6.9849],[232,-7.309],[233,-7.6071],[234,-7.8836],[235,-8.1405],[236,-8.3652],[237,-8.5627],[238,-8.7198],[239,-8.8404],[240,-8.9132],[241,-8.9443],[242,-8.9254],[243,-8.8505],[244,-8.7184],[245,-8.5551],[248,-7.9908],[250,-7.5632],[252,-7.0845],[255,-6.4351],[257,-6.0621],[261,-5.4072],[263,-5.1126],[264,-4.9792],[268,-4.5328],[270,-4.3348],[272,-4.1619],[274,-4.0173],[276,-3.8925],[279,-3.7231],[280,-3.6745],[282,-3.5601],[285,-3.3627],[287,-3.2066],[289,-3.0264],[290,-2.9282],[292,-2.7026],[293,-2.5769],[294,-2.44],[295,-2.2957],[296,-2.1337],[297,-1.9568],[298,-1.7684],[299,-1.5686],[300,-1.3594],[301,-1.137],[302,-0.9091],[303,-0.6758],[304,-0.4311],[305,-0.1855],[306,0.0594],[307,0.3117],[308,0.5661],[311,1.3089],[312,1.5506],[314,2.0122],[316,2.4512],[317,2.6587],[318,2.8567],[319,3.0488],[320,3.2335],[354,-1.1277]]
       }
       foot_r_ik_ctrl.Location.Z {
        type=MovieSceneScriptingFloatChannel
        default=0
        thinned_from=322
        keys [[0,43.1527],[1,43.304],[2,43.1511],[5,42.8647],[8,42.2736],[20,39.1272],[28,37.6181],[31,37.2906],[35,37.285],[47,38.5417],[62,40.3897],[68,40.7192],[73,40.4937],[78,39.7958],[83,38.7289],[92,35.9637],[95,35.2243],[98,34.7318],[101,34.4743],[105,34.4503],[108,34.7283],[111,35.3004],[114,36.1618],[126,40.7728],[130,41.9481],[134,42.7896],[138,43.2647],[142,43.3283],[185,37.5757],[197,36.8148],[203,36.8109],[211,37.4442],[233,41.3599],[240,42.6783],[248,43.2455],[257,43.4884],[284,45.6733],[295,45.6867],[306,44.7665],[320,42.5347],[354,43.1527]]
       }
       foot_r_ik_ctrl.Rotation.X {
        type=MovieSceneScriptingFloatChannel
        default=0
        thinned_from=322
        keys [[0,57.7312],[1,56.5235],[2,55.1108],[4,53.3125],[9,49.7984],[12,48.0045],[14,47.0455],[18,45.6611],[20,45.1811],[23,44.855],[27,44.8621],[35,45.8862],[51,48.7987],[57,49.2351],[64,48.8878],[75,47.0836],[80,45.7947],[82,45.0527],[83,44.5478],[84,43.8991],[85,43.1288],[86,42.0795],[87,40.6489],[88,38.9212],[89,36.9687],[92,30.638],[93,28.634],[94,26.8193],[95,25.2755],[96,23.9435],[97,22.7277],[98,21.624],[99,20.6541],[100,19.8549],[101,19.2624],[102,18.8698],[103,18.6653],[104,18.6448],[105,18.8036],[106,19.1484],[107,19.6688],[108,20.3366],[109,21.1061],[110,21.9488],[114,25.6277],[117,28.1397],[119,29.6906],[120,30.3856],[121,30.9634],[122,31.3887],[123,31.6464],[124,31.7688],[125,31.758],[126,31.6092],[127,31.3565],[131,30.0248],[132,29.7522],[133,29.5413],[134,29.4184],[135,29.3917],[136,29.4903],[139,29.892],[143,30.1365],[147,30.1406],[160,29.404],[162,29.1551],[164,28.7439],[166,28.1843],[169,27.1272],[177,23.7752],[180,22.6645],[184,21.4466],[188,20.4561],[191,19.8895],[193,19.5992],[195,19.3817],[197,19.242],[199,19.1839],[202,19.2502],[204,19.3423],[209,19.4045],[211,19.5208],[212,19.6305],[213,19.7851],[214,19.991],[215,20.26],[216,20.7294],[217,21.4949],[218,22.5317],[219,23.8107],[220,25.2904],[221,26.9431],[222,28.7172],[225,34.2339],[226,36.0092],[227,37.682],[228,39.2057],[229,40.5254],[231,42.831],[233,44.9031],[235,46.761],[237,48.3897],[239,49.7764],[241,50.9101],[243,51.708],[245,52.1757],[248,52.4534],[255,51.9806],[272,53.5803],[292,58.8681],[297,59.6265],[303,59.7608],[309,59.1007],[320,56.3417],[354,57.7312]]
       }
       foot_r_ik_ctrl.Rotation.Y {
        type=MovieSceneScriptingFloatChannel
        default=0
        thinned_from=322
        keys [[0,-7.5275],[1,-7.0304],[2,-6.7331],[4,-6.3095],[6,-5.9307],[10,-5.2446],[12,-4.936],[13,-4.7893],[14,-4.6523],[15,-4.5274],[16,-4.4254],[17,-4.3419],[18,-4.2705],[19,-4.221],[20,-4.1885],[21,-4.1672],[23,-4.1558],[24,-4.1631],[27,-4.2079],[30,-4.2631],[32,-4.2719],[33,-4.264],[34,-4.2457],[35,-4.2109],[36,-4.1677],[37,-4.1133],[38,-4.0437],[39,-3.9646],[40,-3.8739],[42,-3.6656],[43,-3.5557],[45,-3.305],[47,-3.0382],[48,-2.9096],[50,-2.643],[52,-2.3958],[53,-2.2781],[55,-2.0797],[56,-1.9977],[58,-1.8611],[60,-1.7599],[61,-1.7174],[63,-1.6563],[64,-1.6358],[66,-1.6098],[67,-1.6052],[68,-1.6041],[70,-1.6158],[73,-1.6551],[77,-1.7234],[80,-1.7695],[82,-1.7903],[83,-1.8164],[84,-1.8723],[85,-1.9548],[86,-2.0525],[87,-2.1577],[88,-2.2152],[89,-2.1827],[90,-2.0889],[91,-1.9526],[92,-1.7996],[93,-1.6537],[94,-1.5306],[95,-1.4441],[96,-1.4021],[97,-1.4075],[98,-1.4747],[99,-1.6033],[100,-1.7778],[101,-1.9758],[102,-2.1805],[103,-2.3686],[104,-2.5419],[105,-2.7014],[106,-2.8475],[107,-2.9827],[110,-3.3672],[111,-3.5002],[112,-3.6403],[113,-3.7932],[114,-3.9595],[115,-4.1432],[121,-5.3082],[123,-5.6609],[125,-5.9764],[126,-6.1192],[127,-6.2496],[129,-6.4672],[130,-6.5503],[131,-6.6199],[132,-6.6728],[133,-6.7107],[134,-6.7339],[136,-6.7371],[137,-6.7135],[140,-6.574],[144,-6.3146],[152,-5.7062],[155,-5.5096],[157,-5.4039],[161,-5.2705],[170,-5.0867],[176,-5.0266],[178,-5.0265],[181,-5.056],[187,-5.2096],[198,-5.545],[203,-5.6395],[210,-5.6897],[213,-5.7414],[215,-5.7962],[216,-5.8384],[217,-5.8975],[221,-6.188],[223,-6.3159],[231,-6.9804],[235,-7.3697],[237,-7.594],[240,-8.0399],[243,-8.5693],[254,-10.7134],[256,-11.0578],[258,-11.3314],[259,-11.4318],[260,-11.49],[261,-11.5106],[262,-11.4928],[263,-11.443],[264,-11.3579],[265,-11.2396],[266,-11.0916],[268,-10.729],[271,-10.0809],[274,-9.3325],[280,-7.7167],[282,-7.2252],[283,-6.9964],[284,-6.7831],[285,-6.5907],[286,-6.415],[287,-6.2674],[288,-6.1386],[289,-6.0343],[290,-5.9514],[292,-5.8235],[293,-5.777],[295,-5.7191],[296,-5.7091],[298,-5.7305],[299,-5.7589],[301,-5.8535],[303,-5.9898],[305,-6.1586],[307,-6.3596],[309,-6.5914],[320,-8.0934],[354,-7.5275]]
       }
       foot_r_ik_ctrl.Rotation.Z {
        type=MovieSceneScriptingFloatChannel
        default=0
        thinned_from=322
        keys [[0,8.1723],[1,8.849],[2,9.0039],[4,9.4515],[7,10.0576],[11,10.7769],[13,11.0801],[15,11.3441],[16,11.5002],[18,11.8873],[25,13.4715],[27,13.8674],[29,14.1918],[31,14.4083],[32,14.4691],[33,14.4917],[34,14.4807],[35,14.4206],[36,14.3302],[37,14.2092],[39,13.8697],[41,13.434],[43,12.9371],[49,11.2771],[51,10.7616],[53,10.2972],[55,9.9179],[57,9.6364],[59,9.4117],[61,9.2337],[63,9.1035],[66,8.9836],[69,8.95],[72,8.9882],[77,9.1674],[78,9.2233],[80,9.3942],[81,9.511],[82,9.6471],[83,9.8099],[84,10.0105],[85,10.2449],[86,10.5201],[88,11.1902],[89,11.4797],[90,11.7208],[91,11.9011],[92,12.0237],[93,12.0899],[94,12.1125],[97,12.062],[98,12.0737],[99,12.1296],[100,12.2477],[101,12.4467],[105,13.3816],[107,13.7773],[123,16.2737],[131,17.7975],[134,18.2429],[136,18.4245],[158,19.6739],[167,20.5028],[193,24.3828],[209,25.987],[215,26.2887],[217,26.2738],[218,26.2217],[219,26.1083],[220,25.9036],[221,25.6104],[222,25.2261],[223,24.7607],[224,24.2264],[225,23.6173],[226,22.9433],[227,22.2005],[228,21.398],[229,20.5391],[230,19.6242],[232,18.3186],[233,17.6178],[235,16.1192],[236,15.3292],[237,14.4931],[238,13.5828],[239,12.6048],[240,11.5694],[241,10.4924],[242,9.3747],[244,7.0697],[246,4.7235],[247,3.5577],[248,2.4047],[249,1.282],[250,0.1872],[251,-0.8605],[252,-1.8649],[253,-2.8152],[254,-3.7067],[255,-4.5632],[256,-5.3856],[257,-6.175],[258,-6.9268],[259,-7.6372],[260,-8.304],[261,-8.9298],[262,-9.5002],[263,-10.0176],[264,-10.4727],[265,-10.8631],[266,-11.1867],[270,-12.2756],[273,-13.0128],[276,-13.6759],[279,-14.247],[281,-14.5666],[283,-14.8293],[286,-15.1191],[289,-15.2942],[293,-15.3905],[295,-15.3479],[298,-15.1589],[302,-14.7328],[311,-13.5361],[320,-12.4487],[354,8.1723]]
       }
       foot_r_ik_ctrl.Scale.X {type=MovieSceneScriptingFloatChannel, default=1, const [1,322]}
       foot_r_ik_ctrl.Scale.Y {type=MovieSceneScriptingFloatChannel, default=1, const [1,322]}
       foot_r_ik_ctrl.Scale.Z {type=MovieSceneScriptingFloatChannel, default=1, const [1,322]}
       ball_r_ik_ctrl.Location.X {type=MovieSceneScriptingFloatChannel, default=0, thinned_from=322, keys [[0,0],[354,0]]}
       ball_r_ik_ctrl.Location.Y {type=MovieSceneScriptingFloatChannel, default=0, thinned_from=322, keys [[0,0],[354,0]]}
       ball_r_ik_ctrl.Location.Z {type=MovieSceneScriptingFloatChannel, default=0, thinned_from=322, keys [[0,0],[354,0]]}
       ball_r_ik_ctrl.Rotation.X {
        type=MovieSceneScriptingFloatChannel
        default=0
        thinned_from=322
        keys [[0,0.549],[1,0.5323],[2,0.5323],[3,0.5155],[4,0.5155],[5,0.4985],[6,0.4985],[7,0.4819],[9,0.4819],[10,0.4651],[12,0.4651],[13,0.4486],[15,0.4486],[16,0.4319],[19,0.4327],[20,0.421],[22,0.421],[23,0.4142],[28,0.4148],[29,0.421],[32,0.421],[33,0.4187],[34,0.4187],[35,0.4338],[40,0.4338],[41,0.4424],[44,0.4424],[45,0.4539],[47,0.4539],[48,0.4595],[49,0.4585],[51,0.4585],[52,0.4635],[56,0.4635],[57,0.4684],[64,0.4684],[65,0.47],[66,0.4644],[74,0.4644],[75,0.4539],[80,0.4539],[81,0.4525],[83,0.4525],[84,0.4423],[86,0.4423],[87,0.4217],[88,0.4217],[89,0.3961],[92,0.3351],[93,0.3198],[94,0.3198],[95,0.2998],[96,0.2998],[97,0.2895],[99,0.2895],[100,0.2805],[117,0.2807],[121,0.2803],[122,0.2691],[123,0.2691],[124,0.2479],[125,0.2479],[126,0.2267],[127,0.2267],[128,0.2053],[129,0.2095],[130,0.1858],[131,0.1858],[132,0.1701],[133,0.1701],[134,0.1543],[160,0.1545],[161,0.1517],[171,0.1517],[172,0.1382],[176,0.1382],[177,0.1423],[178,0.1207],[186,0.1207],[187,0.1099],[192,0.1099],[193,0.1113],[208,0.1113],[209,0.1151],[210,0.0957],[217,0.0957],[224,0.2311],[225,0.2513],[228,0.3031],[229,0.3117],[230,0.3117],[231,0.329],[232,0.329],[233,0.3463],[234,0.3463],[235,0.3636],[237,0.3636],[238,0.3806],[240,0.3806],[241,0.3882],[242,0.3882],[243,0.3918],[244,0.3918],[245,0.3926],[248,0.392],[249,0.3903],[250,0.3903],[252,0.3865],[253,0.3865],[254,0.3839],[255,0.3839],[256,0.3829],[257,0.3828],[258,0.3814],[259,0.3814],[260,0.3837],[262,0.3837],[263,0.3858],[268,0.3858],[269,0.3993],[272,0.3993],[273,0.4093],[276,0.4093],[277,0.4269],[279,0.4269],[280,0.4445],[282,0.4445],[283,0.4621],[285,0.4621],[286,0.4797],[287,0.4797],[288,0.4914],[289,0.4931],[290,0.5066],[292,0.5066],[293,0.5201],[295,0.5201],[296,0.5337],[298,0.5337],[299,0.5427],[303,0.5427],[304,0.5563],[305,0.5581],[319,0.5581],[320,0.5423],[354,0.549]]
       }
       ball_r_ik_ctrl.Rotation.Y {
        type=MovieSceneScriptingFloatChannel
        default=0
        thinned_from=322
        keys [[0,-0.1434],[1,-0.1406],[2,-0.1406],[3,-0.1391],[6,-0.1388],[7,-0.1345],[9,-0.1345],[10,-0.1326],[12,-0.1326],[13,-0.1266],[15,-0.1266],[16,-0.1231],[17,-0.1244],[19,-0.1244],[20,-0.1109],[22,-0.1109],[23,-0.1139],[25,-0.1139],[26,-0.1066],[28,-0.1066],[29,-0.1078],[34,-0.1078],[35,-0.1063],[40,-0.1063],[41,-0.1118],[44,-0.1118],[45,-0.1176],[47,-0.1176],[48,-0.1228],[49,-0.1225],[51,-0.1225],[52,-0.1258],[56,-0.1258],[57,-0.1307],[64,-0.1307],[65,-0.1326],[66,-0.1361],[74,-0.1361],[75,-0.1315],[80,-0.1315],[81,-0.1337],[83,-0.1337],[84,-0.1313],[86,-0.1313],[87,-0.129],[88,-0.129],[89,-0.1236],[90,-0.1194],[91,-0.1136],[92,-0.1082],[93,-0.1047],[94,-0.1047],[95,-0.097],[96,-0.097],[97,-0.091],[99,-0.091],[100,-0.0868],[112,-0.0868],[113,-0.0926],[117,-0.0926],[118,-0.0957],[121,-0.0957],[122,-0.1008],[123,-0.1008],[124,-0.1005],[125,-0.1005],[126,-0.1002],[127,-0.1002],[128,-0.1009],[130,-0.1002],[131,-0.1002],[132,-0.0994],[144,-0.0995],[145,-0.0987],[160,-0.0987],[171,-0.0986],[172,-0.0975],[177,-0.0975],[178,-0.0955],[186,-0.0955],[187,-0.0943],[208,-0.0942],[209,-0.0945],[210,-0.0935],[217,-0.0935],[219,-0.0977],[222,-0.1044],[223,-0.107],[224,-0.1088],[225,-0.1111],[226,-0.1131],[227,-0.1147],[228,-0.1167],[229,-0.1178],[230,-0.1178],[231,-0.1192],[232,-0.1192],[233,-0.1211],[234,-0.1211],[235,-0.1231],[237,-0.1231],[238,-0.1273],[240,-0.1273],[241,-0.1397],[242,-0.1397],[243,-0.1534],[244,-0.1534],[245,-0.1713],[246,-0.1713],[247,-0.189],[248,-0.189],[249,-0.2098],[250,-0.2098],[251,-0.2243],[252,-0.2368],[253,-0.2368],[254,-0.2589],[255,-0.2589],[256,-0.2766],[257,-0.2747],[258,-0.2918],[259,-0.2918],[260,-0.3072],[262,-0.3072],[263,-0.3246],[268,-0.3246],[269,-0.3319],[272,-0.3319],[273,-0.3277],[292,-0.3291],[303,-0.3296],[304,-0.3289],[319,-0.3293],[320,-0.3288],[354,-0.1434]]
       }
       ball_r_ik_ctrl.Rotation.Z {
        type=MovieSceneScriptingFloatChannel
        default=0
        thinned_from=322
        keys [[0,-0.1031],[1,-0.1044],[4,-0.1044],[5,-0.1033],[6,-0.1033],[7,-0.1003],[9,-0.1003],[10,-0.095],[12,-0.095],[13,-0.0879],[15,-0.0879],[16,-0.0783],[17,-0.0808],[19,-0.0808],[20,-0.0623],[22,-0.0623],[23,-0.0486],[25,-0.0486],[26,-0.0352],[28,-0.0352],[29,-0.0243],[32,-0.0243],[33,-0.0249],[34,-0.0249],[35,-0.0159],[40,-0.0158],[41,-0.0308],[44,-0.0308],[45,-0.0458],[47,-0.0458],[48,-0.0607],[49,-0.0592],[51,-0.0592],[52,-0.0745],[56,-0.0745],[57,-0.0882],[64,-0.0882],[65,-0.0888],[66,-0.1069],[74,-0.1069],[75,-0.1032],[80,-0.1032],[81,-0.1043],[83,-0.1043],[84,-0.1049],[86,-0.1049],[87,-0.1045],[88,-0.1045],[89,-0.1077],[90,-0.1117],[91,-0.1181],[92,-0.125],[93,-0.1312],[94,-0.1312],[95,-0.1461],[96,-0.1461],[97,-0.16],[99,-0.16],[100,-0.1731],[112,-0.1731],[113,-0.1606],[117,-0.1606],[118,-0.1487],[121,-0.1487],[122,-0.1362],[123,-0.1362],[124,-0.1304],[125,-0.1304],[126,-0.1245],[127,-0.1245],[128,-0.1198],[129,-0.1202],[130,-0.1148],[131,-0.1148],[132,-0.1106],[133,-0.1106],[134,-0.1073],[144,-0.1073],[145,-0.1077],[160,-0.1077],[161,-0.108],[171,-0.108],[172,-0.106],[176,-0.106],[177,-0.1058],[186,-0.1059],[187,-0.105],[192,-0.105],[208,-0.1051],[209,-0.1054],[210,-0.1048],[217,-0.1048],[219,-0.1039],[220,-0.1044],[221,-0.104],[222,-0.1044],[224,-0.1045],[226,-0.1054],[227,-0.1053],[229,-0.1062],[232,-0.1063],[233,-0.1067],[234,-0.1067],[235,-0.1071],[237,-0.1071],[238,-0.1087],[240,-0.1087],[241,-0.1172],[242,-0.1172],[243,-0.127],[244,-0.127],[245,-0.14],[246,-0.14],[247,-0.153],[248,-0.153],[249,-0.1706],[250,-0.1706],[251,-0.1791],[252,-0.1895],[253,-0.1895],[254,-0.2058],[255,-0.2058],[256,-0.2188],[257,-0.2169],[258,-0.2285],[259,-0.2285],[260,-0.242],[262,-0.242],[263,-0.2536],[268,-0.2536],[269,-0.2618],[272,-0.2618],[273,-0.2609],[276,-0.2609],[277,-0.263],[279,-0.263],[280,-0.2651],[282,-0.2651],[283,-0.267],[285,-0.267],[286,-0.2689],[287,-0.2689],[288,-0.2705],[289,-0.2702],[290,-0.2718],[292,-0.2718],[293,-0.2728],[295,-0.2728],[296,-0.2744],[298,-0.2744],[299,-0.2755],[303,-0.2755],[304,-0.2773],[319,-0.2771],[320,-0.2766],[354,-0.1031]]
       }
       ball_r_ik_ctrl.Scale.X {type=MovieSceneScriptingFloatChannel, default=1, const [1,322]}
       ball_r_ik_ctrl.Scale.Y {type=MovieSceneScriptingFloatChannel, default=1, const [1,322]}
       ball_r_ik_ctrl.Scale.Z {type=MovieSceneScriptingFloatChannel, default=1, const [1,322]}
       foot_bk1_r_ctrl.Location.X {type=MovieSceneScriptingFloatChannel, default=0, const [0,322]}
       foot_bk1_r_ctrl.Location.Y {type=MovieSceneScriptingFloatChannel, default=0, const [0,322]}
       foot_bk1_r_ctrl.Location.Z {type=MovieSceneScriptingFloatChannel, default=0, const [0,322]}
       foot_bk1_r_ctrl.Rotation.X {type=MovieSceneScriptingFloatChannel, default=0, const [0,322]}
       foot_bk1_r_ctrl.Rotation.Y {type=MovieSceneScriptingFloatChannel, default=0, const [0,322]}
       foot_bk1_r_ctrl.Rotation.Z {type=MovieSceneScriptingFloatChannel, default=0, const [0,322]}
       foot_bk1_r_ctrl.Scale.X {type=MovieSceneScriptingFloatChannel, default=1, const [1,322]}
       foot_bk1_r_ctrl.Scale.Y {type=MovieSceneScriptingFloatChannel, default=1, const [1,322]}
       foot_bk1_r_ctrl.Scale.Z {type=MovieSceneScriptingFloatChannel, default=1, const [1,322]}
       foot_roll_r_ctrl.Location.X {type=MovieSceneScriptingFloatChannel, default=0, const [0,322]}
       foot_roll_r_ctrl.Location.Y {type=MovieSceneScriptingFloatChannel, default=0, const [0,322]}
       foot_roll_r_ctrl.Location.Z {type=MovieSceneScriptingFloatChannel, default=0, const [0,322]}
       foot_roll_r_ctrl.Rotation.X {type=MovieSceneScriptingFloatChannel, default=0, const [0,322]}
       foot_roll_r_ctrl.Rotation.Y {type=MovieSceneScriptingFloatChannel, default=0, const [0,322]}
       foot_roll_r_ctrl.Rotation.Z {type=MovieSceneScriptingFloatChannel, default=0, const [0,322]}
       foot_roll_r_ctrl.Scale.X {type=MovieSceneScriptingFloatChannel, default=1, const [1,322]}
       foot_roll_r_ctrl.Scale.Y {type=MovieSceneScriptingFloatChannel, default=1, const [1,322]}
       foot_roll_r_ctrl.Scale.Z {type=MovieSceneScriptingFloatChannel, default=1, const [1,322]}
       roll_blend_r_ctrl {type=MovieSceneScriptingFloatChannel, default=1, const [1,322]}
       tip_r_ctrl.Location.X {type=MovieSceneScriptingFloatChannel, default=0, const [0,322]}
       tip_r_ctrl.Location.Y {type=MovieSceneScriptingFloatChannel, default=0, const [0,322]}
       tip_r_ctrl.Location.Z {type=MovieSceneScriptingFloatChannel, default=0, const [0,322]}
       tip_r_ctrl.Rotation.X {type=MovieSceneScriptingFloatChannel, default=0, const [0,322]}
       tip_r_ctrl.Rotation.Y {type=MovieSceneScriptingFloatChannel, default=0, const [0,322]}
       tip_r_ctrl.Rotation.Z {type=MovieSceneScriptingFloatChannel, default=0, const [0,322]}
       tip_r_ctrl.Scale.X {type=MovieSceneScriptingFloatChannel, default=1, const [1,322]}
       tip_r_ctrl.Scale.Y {type=MovieSceneScriptingFloatChannel, default=1, const [1,322]}
       tip_r_ctrl.Scale.Z {type=MovieSceneScriptingFloatChannel, default=1, const [1,322]}
       heel_r_ctrl.Location.X {type=MovieSceneScriptingFloatChannel, default=0, const [0,322]}
       heel_r_ctrl.Location.Y {type=MovieSceneScriptingFloatChannel, default=0, const [0,322]}
       heel_r_ctrl.Location.Z {type=MovieSceneScriptingFloatChannel, default=0, const [0,322]}
       heel_r_ctrl.Rotation.X {type=MovieSceneScriptingFloatChannel, default=0, const [0,322]}
       heel_r_ctrl.Rotation.Y {type=MovieSceneScriptingFloatChannel, default=0, const [0,322]}
       heel_r_ctrl.Rotation.Z {type=MovieSceneScriptingFloatChannel, default=0, const [0,322]}
       heel_r_ctrl.Scale.X {type=MovieSceneScriptingFloatChannel, default=1, const [1,322]}
       heel_r_ctrl.Scale.Y {type=MovieSceneScriptingFloatChannel, default=1, const [1,322]}
       heel_r_ctrl.Scale.Z {type=MovieSceneScriptingFloatChannel, default=1, const [1,322]}
       leg_r_softness {type=MovieSceneScriptingFloatChannel, default=0, const [0,322]}
       leg_r_pv_ik_ctrl.Location.X {
        type=MovieSceneScriptingFloatChannel
        default=0
        thinned_from=322
        keys [[0,-5.6082],[1,-5.4606],[2,-5.2401],[3,-5.1245],[6,-4.7335],[8,-4.4369],[9,-3.445],[10,-4.1104],[14,-3.4133],[15,-3.2301],[16,-3.0302],[17,-2.8058],[18,-2.5685],[19,-2.3243],[20,-2.0676],[21,-1.8184],[22,-1.5735],[23,-1.3422],[24,-1.1139],[25,-0.8766],[26,-0.6591],[27,-0.456],[28,-0.2692],[29,-0.1068],[30,0.0273],[31,0.1342],[32,0.21],[33,0.2472],[34,0.252],[35,0.2122],[36,0.1412],[37,0.0403],[38,-0.0953],[39,-0.2507],[40,-0.4315],[41,-0.632],[42,-0.8472],[43,-1.0732],[45,-1.5623],[47,-2.0696],[49,-2.5688],[50,-2.8081],[51,-3.0381],[52,-3.2574],[53,-3.4616],[54,-3.6428],[55,-3.8094],[56,-3.9472],[57,-4.069],[58,-4.1805],[59,-4.2795],[61,-4.4511],[63,-4.582],[66,-4.7148],[68,-4.7608],[70,-4.7819],[72,-4.7785],[76,-4.7214],[78,-4.6755],[80,-4.6097],[82,-4.5177],[83,-4.4525],[84,-4.3647],[85,-4.2564],[86,-4.1148],[88,-3.7521],[89,-3.6132],[90,-3.5199],[91,-3.4771],[92,-3.4859],[93,-3.5398],[94,-3.6325],[95,-3.754],[97,-4.0142],[98,-4.1237],[99,-4.2031],[100,-4.237],[101,-4.2004],[104,-3.9967],[105,-3.9385],[106,-3.893],[107,-3.8578],[109,-3.8125],[115,-3.7288],[118,-3.6657],[120,-3.606],[121,-3.5634],[122,-3.5107],[123,-3.4436],[124,-3.3672],[125,-3.2806],[126,-3.1806],[129,-2.8476],[130,-2.7437],[131,-2.6455],[132,-2.5601],[133,-2.489],[134,-2.4361],[135,-2.4006],[136,-2.3865],[137,-2.3909],[147,-2.4886],[155,-2.5255],[160,-2.523],[162,-2.5086],[164,-2.4794],[166,-2.4359],[168,-2.3797],[172,-2.2414],[179,-1.9821],[184,-1.8246],[186,-1.7671],[189,-1.6911],[192,-1.6281],[195,-1.5797],[197,-1.5565],[199,-1.5421],[201,-1.5362],[205,-1.5397],[210,-1.5365],[212,-1.5434],[213,-1.5502],[214,-1.5603],[215,-1.5744],[216,-1.6006],[217,-1.6444],[218,-1.7054],[219,-1.7821],[220,-1.8738],[221,-1.9798],[222,-2.099],[223,-2.2286],[224,-2.3672],[227,-2.807],[228,-2.9491],[229,-3.0811],[234,-3.6925],[236,-3.9196],[237,-4.0704],[238,-4.2982],[239,-4.6061],[240,-4.9821],[241,-5.4205],[242,-5.9147],[243,-6.4613],[244,-7.0492],[245,-7.6759],[246,-8.3343],[254,-13.8244],[256,-15.1168],[258,-16.3046],[259,-16.8396],[260,-17.3301],[261,-17.7743],[262,-18.1551],[263,-18.4757],[264,-18.7232],[265,-18.8933],[266,-18.9826],[290,-20.2083],[307,-20.303],[320,-19.9604],[354,-5.6082]]
       }
       leg_r_pv_ik_ctrl.Location.Y {
        type=MovieSceneScriptingFloatChannel
        default=0
        thinned_from=322
        keys [[0,11.3236],[1,11.1976],[2,11.0272],[8,10.6412],[9,5.6653],[10,10.5314],[13,10.4172],[15,10.3888],[20,10.3527],[24,10.4871],[35,10.5458],[53,10.8826],[70,10.8345],[77,10.7041],[80,10.5761],[82,10.4479],[83,10.3521],[84,10.2174],[85,10.0476],[86,9.8443],[87,9.6113],[88,9.3442],[89,9.0545],[90,8.7405],[91,8.4047],[93,7.6782],[97,6.1148],[98,5.7327],[99,5.3663],[100,5.023],[101,4.6974],[102,4.421],[103,4.1879],[104,4.0071],[105,3.8862],[106,3.8381],[107,3.8678],[108,3.9636],[109,4.1082],[110,4.2942],[111,4.5154],[112,4.7632],[113,5.0263],[118,6.4211],[120,6.9421],[122,7.4043],[123,7.606],[124,7.7886],[125,7.9471],[126,8.0773],[127,8.181],[130,8.4112],[133,8.5891],[138,8.7871],[145,8.9257],[157,8.9456],[162,8.8729],[165,8.77],[170,8.4964],[188,7.2763],[193,7.0253],[197,6.897],[199,6.8626],[202,6.8557],[210,6.9784],[214,7.1002],[217,7.2486],[219,7.3826],[221,7.5547],[223,7.7636],[225,8.0195],[228,8.4899],[237,10.1126],[239,10.4333],[241,10.7059],[243,10.9139],[245,11.0417],[246,11.0694],[247,11.0725],[248,11.0468],[249,10.993],[251,10.8012],[259,9.837],[261,9.6291],[263,9.4736],[264,9.4238],[265,9.3992],[286,9.4997],[320,8.8424],[354,11.3236]]
       }
       leg_r_pv_ik_ctrl.Location.Z {
        type=MovieSceneScriptingFloatChannel
        default=0
        thinned_from=322
        keys [[0,14.7111],[1,14.4703],[2,14.1496],[8,13.4489],[9,6.0017],[10,13.2598],[12,13.1197],[14,13.0497],[20,13.0665],[24,13.41],[57,13.9314],[70,13.794],[76,13.6092],[78,13.499],[80,13.3312],[82,13.1057],[83,12.9388],[84,12.7065],[85,12.4174],[86,12.0768],[87,11.6945],[88,11.2645],[89,10.8062],[90,10.3184],[92,9.2742],[96,7.0899],[97,6.5665],[98,6.0688],[99,5.6025],[100,5.1749],[101,4.778],[102,4.4472],[103,4.1723],[104,3.9619],[105,3.8226],[106,3.768],[107,3.8032],[108,3.915],[109,4.0845],[110,4.304],[111,4.5679],[112,4.8672],[113,5.1896],[115,5.8857],[119,7.3401],[120,7.6899],[122,8.3395],[123,8.63],[124,8.897],[125,9.1325],[126,9.3293],[127,9.489],[130,9.8544],[133,10.1433],[136,10.3529],[141,10.5716],[149,10.7119],[157,10.6944],[161,10.6155],[164,10.4866],[168,10.1949],[186,8.4427],[191,8.0542],[195,7.834],[197,7.7603],[199,7.7143],[202,7.7054],[210,7.8749],[213,7.9914],[215,8.1005],[218,8.3253],[220,8.5273],[222,8.7828],[223,8.9296],[224,9.0968],[225,9.2838],[227,9.7202],[229,10.2173],[233,11.3322],[237,12.5323],[239,13.0794],[241,13.5689],[243,13.9873],[245,14.321],[247,14.5492],[248,14.6154],[249,14.6465],[255,14.5319],[264,14.7107],[283,15.5149],[290,15.5943],[301,15.3106],[320,14.3175],[354,14.7111]]
       }
       leg_r_pv_ik_ctrl.Rotation.X {
        type=MovieSceneScriptingFloatChannel
        default=0
        thinned_from=322
        keys [[0,-18.1634],[1,-17.8955],[2,-17.5376],[8,-16.7497],[9,-8.0794],[10,-16.5355],[12,-16.3759],[14,-16.2952],[20,-16.3076],[24,-16.6897],[58,-17.2892],[74,-17.0154],[78,-16.8072],[81,-16.5008],[82,-16.364],[83,-16.1754],[84,-15.9122],[85,-15.5836],[86,-15.1951],[87,-14.7569],[88,-14.2617],[89,-13.7312],[90,-13.1633],[91,-12.5631],[93,-11.2897],[96,-9.3104],[97,-8.6681],[98,-8.0521],[99,-7.4699],[100,-6.9317],[101,-6.4279],[102,-6.0049],[103,-5.6513],[104,-5.3792],[105,-5.1984],[106,-5.1272],[107,-5.1727],[108,-5.3175],[109,-5.5363],[110,-5.819],[111,-6.1571],[112,-6.5386],[113,-6.9472],[115,-7.8214],[118,-9.176],[120,-10.0394],[122,-10.8222],[123,-11.1696],[124,-11.4876],[125,-11.7669],[126,-11.9996],[127,-12.1876],[130,-12.6161],[133,-12.9532],[136,-13.1976],[141,-13.4525],[149,-13.6159],[158,-13.5786],[162,-13.4649],[165,-13.282],[170,-12.8016],[186,-10.9363],[191,-10.4687],[195,-10.2025],[198,-10.0829],[201,-10.0416],[209,-10.2174],[213,-10.3925],[215,-10.5241],[218,-10.7952],[220,-11.0381],[222,-11.3446],[223,-11.5201],[225,-11.942],[227,-12.4586],[229,-13.0429],[233,-14.3387],[237,-15.7132],[239,-16.3348],[241,-16.8895],[243,-17.3631],[245,-17.7415],[247,-18.0023],[248,-18.0796],[249,-18.1181],[255,-18.0158],[264,-18.2417],[283,-19.1358],[290,-19.2248],[301,-18.9144],[320,-17.8142],[354,-18.1634]]
       }
       leg_r_pv_ik_ctrl.Rotation.Y {
        type=MovieSceneScriptingFloatChannel
        default=0
        thinned_from=322
        keys [[0,0.0812],[1,0.0795],[2,0.0785],[3,0.0807],[4,0.0826],[5,0.0848],[6,0.0872],[7,0.0904],[8,0.0944],[9,-1.3056],[10,0.1046],[11,0.1108],[12,0.1176],[13,0.1252],[14,0.1335],[15,0.1428],[17,0.1651],[18,0.1771],[19,0.1896],[20,0.2028],[22,0.2336],[23,0.2482],[24,0.262],[25,0.2743],[26,0.2856],[27,0.2961],[28,0.3059],[29,0.3145],[30,0.3218],[31,0.3278],[32,0.3322],[33,0.3349],[34,0.3359],[35,0.3348],[36,0.3323],[37,0.3283],[38,0.3226],[39,0.316],[40,0.3081],[41,0.2993],[43,0.2793],[45,0.2566],[49,0.2084],[50,0.1966],[52,0.1742],[53,0.1639],[54,0.1546],[55,0.1462],[56,0.139],[57,0.1328],[58,0.1269],[59,0.1216],[61,0.1123],[62,0.1085],[63,0.1045],[64,0.1015],[66,0.096],[67,0.0939],[68,0.0921],[69,0.0905],[71,0.0881],[73,0.0868],[75,0.0859],[77,0.0844],[78,0.0834],[80,0.0804],[82,0.0767],[83,0.0739],[84,0.0698],[86,0.06],[87,0.056],[88,0.0505],[89,0.0421],[90,0.0306],[91,0.0159],[92,-0.0018],[93,-0.022],[94,-0.0442],[96,-0.0912],[97,-0.1137],[98,-0.1341],[99,-0.1516],[100,-0.1649],[101,-0.1729],[102,-0.1767],[103,-0.1792],[104,-0.1801],[105,-0.1797],[106,-0.1781],[107,-0.1751],[108,-0.1709],[109,-0.1657],[110,-0.1595],[111,-0.1522],[112,-0.1442],[113,-0.1351],[114,-0.1255],[115,-0.1153],[116,-0.1043],[117,-0.093],[118,-0.0814],[119,-0.0695],[121,-0.045],[124,-0.0071],[125,0.0051],[126,0.0168],[127,0.0279],[128,0.0383],[129,0.0481],[130,0.0574],[131,0.066],[132,0.0737],[133,0.0803],[134,0.0856],[135,0.0897],[136,0.0925],[137,0.094],[139,0.0959],[142,0.0976],[143,0.098],[145,0.0984],[149,0.0983],[156,0.0968],[159,0.0954],[163,0.0929],[165,0.0911],[180,0.0734],[185,0.0673],[188,0.0641],[191,0.0614],[194,0.0594],[197,0.0581],[199,0.0576],[201,0.0575],[203,0.0579],[205,0.0589],[210,0.0623],[215,0.0666],[216,0.0672],[217,0.0669],[218,0.0661],[219,0.0647],[220,0.063],[221,0.0609],[224,0.0537],[225,0.0517],[226,0.0503],[227,0.0496],[228,0.0498],[229,0.051],[230,0.0532],[231,0.0558],[232,0.059],[233,0.0625],[234,0.0669],[236,0.0765],[237,0.0792],[238,0.0775],[239,0.0709],[240,0.0601],[241,0.0452],[242,0.0264],[243,0.0036],[244,-0.0229],[245,-0.053],[246,-0.0867],[247,-0.1233],[248,-0.1631],[249,-0.2052],[250,-0.2504],[251,-0.2967],[254,-0.4388],[255,-0.4856],[256,-0.5309],[257,-0.5749],[258,-0.617],[259,-0.6562],[260,-0.6922],[261,-0.7251],[262,-0.7532],[263,-0.7767],[264,-0.7944],[265,-0.8056],[266,-0.8102],[277,-0.83],[308,-0.9061],[320,-0.915],[354,0.0812]]
       }
       leg_r_pv_ik_ctrl.Rotation.Z {
        type=MovieSceneScriptingFloatChannel
        default=0
        thinned_from=322
        keys [[0,6.2572],[1,6.1084],[2,5.8834],[3,5.7624],[6,5.3522],[8,5.0373],[9,3.2651],[10,4.687],[14,3.9296],[15,3.7288],[16,3.509],[17,3.2629],[18,3.0022],[19,2.7339],[20,2.4516],[21,2.1769],[22,1.9072],[23,1.6528],[24,1.4021],[25,1.1415],[26,0.9026],[27,0.6797],[28,0.4747],[29,0.2965],[30,0.1494],[31,0.0323],[32,-0.0506],[33,-0.0912],[34,-0.096],[35,-0.0521],[36,0.0261],[37,0.1371],[38,0.2861],[39,0.4567],[40,0.6549],[41,0.8747],[42,1.1102],[43,1.3574],[45,1.8919],[47,2.4455],[49,2.9894],[50,3.25],[51,3.5003],[52,3.7389],[53,3.9609],[54,4.1579],[55,4.3389],[56,4.4886],[57,4.6207],[58,4.7419],[59,4.8495],[61,5.0362],[63,5.1789],[66,5.3245],[68,5.3759],[70,5.4005],[72,5.3987],[76,5.3416],[79,5.2652],[81,5.1883],[82,5.1381],[83,5.0728],[84,4.9849],[85,4.8763],[86,4.732],[88,4.3569],[89,4.2166],[90,4.1276],[91,4.0964],[92,4.125],[93,4.2072],[94,4.3366],[95,4.5023],[97,4.8627],[98,5.0219],[99,5.1463],[100,5.2156],[101,5.1988],[105,4.9456],[107,4.8471],[115,4.5603],[119,4.3804],[120,4.3267],[121,4.2629],[122,4.189],[123,4.1002],[124,4.0024],[125,3.8949],[127,3.6455],[129,3.3823],[130,3.2613],[131,3.1472],[132,3.0481],[133,2.9657],[134,2.9042],[135,2.8626],[136,2.8456],[137,2.8495],[148,2.962],[155,2.9958],[160,2.9943],[162,2.9792],[164,2.9482],[166,2.9014],[168,2.8408],[172,2.6903],[179,2.405],[184,2.2302],[186,2.1659],[189,2.0807],[192,2.0097],[195,1.9549],[197,1.9286],[199,1.9121],[201,1.9053],[210,1.9038],[212,1.9112],[214,1.93],[215,1.9458],[216,1.9756],[217,2.0258],[218,2.0958],[219,2.1836],[220,2.2885],[221,2.4094],[222,2.5449],[223,2.6916],[225,3.0096],[227,3.3366],[228,3.492],[229,3.6346],[232,4.029],[235,4.3971],[236,4.5106],[237,4.6673],[238,4.9081],[239,5.2359],[240,5.637],[241,6.1047],[242,6.6318],[243,7.2148],[244,7.8415],[245,8.5097],[246,9.2124],[249,11.4195],[254,15.1502],[256,16.5605],[258,17.8581],[259,18.4429],[260,18.9793],[261,19.4651],[262,19.8811],[263,20.2305],[264,20.4986],[265,20.6799],[266,20.7708],[293,22.0796],[315,22.0762],[320,21.9455],[354,6.2572]]
       }
       leg_r_pv_ik_ctrl.Scale.X {type=MovieSceneScriptingFloatChannel, default=1, const [1,322]}
       leg_r_pv_ik_ctrl.Scale.Y {type=MovieSceneScriptingFloatChannel, default=1, const [1,322]}
       leg_r_pv_ik_ctrl.Scale.Z {type=MovieSceneScriptingFloatChannel, default=1, const [1,322]}
       upperarm_l_fk_ctrl.Location.X {type=MovieSceneScriptingFloatChannel, default=0, thinned_from=322, keys [[0,0],[354,0]]}
       upperarm_l_fk_ctrl.Location.Y {type=MovieSceneScriptingFloatChannel, default=0, thinned_from=322, keys [[0,0],[354,0]]}
       upperarm_l_fk_ctrl.Location.Z {type=MovieSceneScriptingFloatChannel, default=0, thinned_from=322, keys [[0,0],[354,0]]}
       upperarm_l_fk_ctrl.Rotation.X {
        type=MovieSceneScriptingFloatChannel
        default=0
        thinned_from=322
        keys [[0,64.3382],[1,64.549],[7,64.2959],[48,60.5594],[84,60.7724],[86,60.5476],[89,59.7552],[92,58.414],[97,55.6798],[99,54.8493],[101,54.2777],[103,53.9331],[105,53.8316],[118,55.4931],[120,55.5752],[122,55.4157],[125,54.7998],[133,52.2157],[136,51.5885],[139,51.284],[172,52.3744],[178,51.9955],[186,51.1121],[200,51.0852],[210,52.1253],[215,53.1551],[218,54.1848],[220,55.0668],[222,56.1852],[234,65.5428],[238,67.8731],[242,69.5497],[247,70.789],[300,72.6655],[320,70.0774],[354,64.3382]]
       }
       upperarm_l_fk_ctrl.Rotation.Y {
        type=MovieSceneScriptingFloatChannel
        default=0
        thinned_from=322
        keys [[0,-14.9498],[1,-14.6243],[8,-12.9189],[10,-12.3768],[11,-12.0504],[12,-11.6766],[13,-11.2644],[14,-10.8193],[16,-9.8597],[19,-8.372],[20,-7.8939],[21,-7.4329],[22,-6.999],[23,-6.5977],[24,-6.2362],[25,-5.9222],[26,-5.6621],[27,-5.4435],[28,-5.256],[29,-5.0973],[30,-4.9722],[31,-4.8755],[32,-4.8037],[33,-4.7562],[34,-4.7271],[36,-4.6916],[39,-4.6804],[44,-4.7245],[51,-4.8118],[52,-4.8358],[53,-4.8364],[57,-4.9199],[83,-5.6617],[84,-5.6595],[85,-5.605],[86,-5.4945],[87,-5.3103],[88,-5.0594],[89,-4.7592],[90,-4.4019],[91,-3.9881],[92,-3.5408],[93,-3.0822],[94,-2.6393],[95,-2.2341],[96,-1.8889],[97,-1.6282],[98,-1.4778],[99,-1.4509],[100,-1.5318],[101,-1.7055],[102,-1.9716],[103,-2.3225],[104,-2.7548],[105,-3.2682],[106,-3.8569],[107,-4.4981],[108,-5.1615],[110,-6.5229],[111,-7.191],[112,-7.8393],[113,-8.4597],[114,-9.0401],[116,-10.1514],[117,-10.6793],[118,-11.1776],[119,-11.6252],[120,-12.0075],[121,-12.311],[122,-12.5313],[123,-12.6789],[124,-12.7773],[125,-12.8188],[126,-12.7989],[127,-12.7275],[129,-12.484],[132,-12.0772],[134,-11.8831],[136,-11.7492],[141,-11.5476],[156,-11.3848],[164,-11.4543],[173,-11.7182],[176,-11.7368],[177,-11.7166],[179,-11.6086],[183,-11.3316],[185,-11.2437],[195,-11.2452],[200,-11.3446],[204,-11.5162],[208,-11.797],[211,-12.093],[213,-12.3363],[214,-12.4795],[215,-12.6505],[216,-12.8608],[217,-13.1132],[218,-13.4006],[219,-13.7283],[220,-14.12],[221,-14.5908],[222,-15.1541],[223,-15.8091],[224,-16.5329],[229,-20.496],[230,-21.2283],[231,-21.8904],[232,-22.4633],[233,-22.9266],[234,-23.2778],[235,-23.5211],[236,-23.6788],[237,-23.778],[238,-23.8278],[239,-23.8245],[241,-23.6754],[243,-23.3542],[245,-22.8961],[248,-22.0185],[252,-20.6517],[261,-17.3025],[263,-16.6349],[264,-16.3561],[265,-16.1202],[266,-15.9328],[267,-15.7956],[268,-15.7017],[294,-15.4315],[297,-15.3038],[299,-15.1498],[301,-14.9228],[306,-14.1485],[318,-11.9449],[320,-11.6261],[354,-14.9498]]
       }
       upperarm_l_fk_ctrl.Rotation.Z {
        type=MovieSceneScriptingFloatChannel
        default=0
        thinned_from=322
        keys [[0,49.1326],[12,49.1991],[33,52.0181],[51,52.3604],[85,49.2974],[87,48.9258],[90,47.9964],[99,44.1381],[103,42.8289],[106,42.1605],[109,41.7954],[115,41.6409],[117,41.7835],[123,42.9776],[130,44.5821],[135,45.1777],[138,45.2374],[162,43.0688],[182,43.0225],[219,45.4977],[226,46.3351],[231,46.3603],[235,46.0059],[246,43.5053],[253,41.6771],[258,40.7982],[265,40.3265],[274,40.427],[289,41.1406],[320,39.0174],[354,49.1326]]
       }
       upperarm_l_fk_ctrl.Scale.X {type=MovieSceneScriptingFloatChannel, default=1, const [1,322]}
       upperarm_l_fk_ctrl.Scale.Y {type=MovieSceneScriptingFloatChannel, default=1, const [1,322]}
       upperarm_l_fk_ctrl.Scale.Z {type=MovieSceneScriptingFloatChannel, default=1, const [1,322]}
       lowerarm_l_fk_ctrl.Location.X {type=MovieSceneScriptingFloatChannel, default=0, thinned_from=322, keys [[0,0],[354,0]]}
       lowerarm_l_fk_ctrl.Location.Y {type=MovieSceneScriptingFloatChannel, default=0, thinned_from=322, keys [[0,0],[354,0]]}
       lowerarm_l_fk_ctrl.Location.Z {type=MovieSceneScriptingFloatChannel, default=0, thinned_from=322, keys [[0,0],[354,0]]}
       lowerarm_l_fk_ctrl.Rotation.X {
        type=MovieSceneScriptingFloatChannel
        default=0
        thinned_from=322
        keys [[0,-0.962],[1,-0.9392],[4,-0.9051],[5,-0.8994],[6,-0.888],[7,-0.8823],[8,-0.8709],[9,-0.8652],[12,-0.8311],[13,-0.8254],[16,-0.7912],[19,-0.768],[23,-0.7489],[27,-0.7351],[28,-0.7351],[29,-0.7287],[31,-0.7287],[32,-0.7213],[35,-0.7217],[36,-0.7104],[38,-0.7104],[39,-0.7048],[42,-0.7048],[43,-0.6991],[44,-0.6991],[45,-0.6935],[47,-0.6935],[48,-0.6878],[49,-0.6866],[50,-0.6823],[51,-0.6823],[54,-0.6696],[55,-0.6696],[62,-0.6398],[63,-0.6312],[66,-0.616],[67,-0.616],[80,-0.5448],[85,-0.5247],[86,-0.5167],[87,-0.5046],[88,-0.4966],[91,-0.4603],[92,-0.4523],[93,-0.4402],[95,-0.4241],[96,-0.4201],[97,-0.4201],[98,-0.4223],[100,-0.4402],[101,-0.4535],[102,-0.4691],[103,-0.4869],[104,-0.507],[106,-0.5516],[109,-0.6251],[110,-0.6518],[113,-0.7243],[115,-0.7684],[116,-0.7867],[117,-0.8032],[118,-0.8161],[119,-0.8252],[120,-0.8326],[121,-0.8344],[122,-0.8326],[123,-0.8271],[124,-0.8234],[126,-0.8124],[133,-0.7862],[134,-0.7862],[138,-0.77],[139,-0.77],[140,-0.7659],[141,-0.7659],[142,-0.7618],[143,-0.7537],[144,-0.7497],[152,-0.7],[153,-0.6909],[157,-0.6665],[158,-0.6635],[159,-0.6574],[161,-0.6513],[162,-0.6513],[163,-0.6483],[164,-0.6483],[165,-0.6452],[166,-0.6452],[174,-0.6209],[175,-0.6148],[176,-0.6118],[177,-0.6118],[178,-0.6037],[180,-0.6037],[181,-0.5955],[186,-0.5955],[187,-0.5874],[197,-0.5874],[198,-0.59],[200,-0.59],[202,-0.5926],[203,-0.5926],[204,-0.5953],[205,-0.5953],[206,-0.5979],[207,-0.5979],[209,-0.6033],[210,-0.6033],[212,-0.6108],[215,-0.6442],[216,-0.6591],[218,-0.6963],[222,-0.7855],[223,-0.8115],[225,-0.849],[228,-0.8928],[231,-0.9147],[233,-0.9147],[235,-0.9001],[238,-0.8563],[240,-0.8125],[241,-0.7933],[243,-0.7446],[244,-0.7243],[247,-0.6513],[249,-0.6108],[251,-0.5783],[252,-0.5661],[253,-0.558],[254,-0.554],[255,-0.554],[256,-0.5418],[257,-0.5406],[258,-0.5382],[261,-0.5382],[263,-0.5431],[265,-0.5518],[266,-0.5553],[267,-0.5627],[268,-0.5676],[272,-0.6069],[280,-0.6927],[283,-0.7201],[285,-0.7338],[287,-0.7433],[290,-0.7517],[293,-0.7517],[294,-0.7487],[295,-0.7487],[297,-0.7426],[298,-0.7426],[300,-0.7365],[304,-0.7365],[305,-0.7334],[310,-0.757],[314,-0.7947],[315,-0.8088],[316,-0.8182],[320,-0.8747],[354,-0.962]]
       }
       lowerarm_l_fk_ctrl.Rotation.Y {type=MovieSceneScriptingFloatChannel, default=0, thinned_from=322, keys [[0,-3.329],[354,-3.329]]}
       lowerarm_l_fk_ctrl.Rotation.Z {type=MovieSceneScriptingFloatChannel, default=0, thinned_from=322, keys [[0,88.5089],[354,88.5089]]}
       lowerarm_l_fk_ctrl.Scale.X {type=MovieSceneScriptingFloatChannel, default=1, const [1,322]}
       lowerarm_l_fk_ctrl.Scale.Y {type=MovieSceneScriptingFloatChannel, default=1, const [1,322]}
       lowerarm_l_fk_ctrl.Scale.Z {type=MovieSceneScriptingFloatChannel, default=1, const [1,322]}
       hand_l_fk_ctrl.Location.X {type=MovieSceneScriptingFloatChannel, default=0, thinned_from=322, keys [[0,0],[354,0]]}
       hand_l_fk_ctrl.Location.Y {type=MovieSceneScriptingFloatChannel, default=0, thinned_from=322, keys [[0,0],[354,0]]}
       hand_l_fk_ctrl.Location.Z {type=MovieSceneScriptingFloatChannel, default=0, thinned_from=322, keys [[0,0],[354,0]]}
       hand_l_fk_ctrl.Rotation.X {
        type=MovieSceneScriptingFloatChannel
        default=0
        thinned_from=322
        keys [[0,22.7788],[101,22.261],[219,22.5285],[236,22.7032],[273,22.4259],[320,22.6899],[354,22.7788]]
       }
       hand_l_fk_ctrl.Rotation.Y {type=MovieSceneScriptingFloatChannel, default=0, thinned_from=322, keys [[0,26.1323],[320,26.1684],[354,26.1323]]}
       hand_l_fk_ctrl.Rotation.Z {type=MovieSceneScriptingFloatChannel, default=0, thinned_from=322, keys [[0,25.4282],[320,25.3887],[354,25.4282]]}
       hand_l_fk_ctrl.Scale.X {type=MovieSceneScriptingFloatChannel, default=1, const [1,322]}
       hand_l_fk_ctrl.Scale.Y {type=MovieSceneScriptingFloatChannel, default=1, const [1,322]}
       hand_l_fk_ctrl.Scale.Z {type=MovieSceneScriptingFloatChannel, default=1, const [1,322]}
       thumb_01_l_ctrl.Location.X {type=MovieSceneScriptingFloatChannel, default=0, const [1.5173,322]}
       thumb_01_l_ctrl.Location.Y {type=MovieSceneScriptingFloatChannel, default=0, const [-0.0715,322]}
       thumb_01_l_ctrl.Location.Z {type=MovieSceneScriptingFloatChannel, default=0, const [0.0001,322]}
       thumb_01_l_ctrl.Rotation.X {type=MovieSceneScriptingFloatChannel, default=0, const [-23.3104,322]}
       thumb_01_l_ctrl.Rotation.Y {type=MovieSceneScriptingFloatChannel, default=0, const [-0.3589,322]}
       thumb_01_l_ctrl.Rotation.Z {type=MovieSceneScriptingFloatChannel, default=0, const [0.0053,322]}
       thumb_01_l_ctrl.Scale.X {type=MovieSceneScriptingFloatChannel, default=1, const [1,322]}
       thumb_01_l_ctrl.Scale.Y {type=MovieSceneScriptingFloatChannel, default=1, const [1,322]}
       thumb_01_l_ctrl.Scale.Z {type=MovieSceneScriptingFloatChannel, default=1, const [1,322]}
       thumb_02_l_ctrl.Location.X {type=MovieSceneScriptingFloatChannel, default=0, thinned_from=322, keys [[0,0],[354,0]]}
       thumb_02_l_ctrl.Location.Y {type=MovieSceneScriptingFloatChannel, default=0, thinned_from=322, keys [[0,0],[354,0]]}
       thumb_02_l_ctrl.Location.Z {type=MovieSceneScriptingFloatChannel, default=0, thinned_from=322, keys [[0,0],[354,0]]}
       thumb_02_l_ctrl.Rotation.X {type=MovieSceneScriptingFloatChannel, default=0, const [1.2228,322]}
       thumb_02_l_ctrl.Rotation.Y {type=MovieSceneScriptingFloatChannel, default=0, const [-2.3382,322]}
       thumb_02_l_ctrl.Rotation.Z {type=MovieSceneScriptingFloatChannel, default=0, const [-13.4617,322]}
       thumb_02_l_ctrl.Scale.X {type=MovieSceneScriptingFloatChannel, default=1, const [1,322]}
       thumb_02_l_ctrl.Scale.Y {type=MovieSceneScriptingFloatChannel, default=1, const [1,322]}
       thumb_02_l_ctrl.Scale.Z {type=MovieSceneScriptingFloatChannel, default=1, const [1,322]}
       thumb_03_l_ctrl.Location.X {type=MovieSceneScriptingFloatChannel, default=0, thinned_from=322, keys [[0,0],[354,0]]}
       thumb_03_l_ctrl.Location.Y {type=MovieSceneScriptingFloatChannel, default=0, thinned_from=322, keys [[0,0],[354,0]]}
       thumb_03_l_ctrl.Location.Z {type=MovieSceneScriptingFloatChannel, default=0, thinned_from=322, keys [[0,0],[354,0]]}
       thumb_03_l_ctrl.Rotation.X {type=MovieSceneScriptingFloatChannel, default=0, const [2.0016,322]}
       thumb_03_l_ctrl.Rotation.Y {type=MovieSceneScriptingFloatChannel, default=0, const [0.1771,322]}
       thumb_03_l_ctrl.Rotation.Z {type=MovieSceneScriptingFloatChannel, default=0, const [-52.3193,322]}
       thumb_03_l_ctrl.Scale.X {type=MovieSceneScriptingFloatChannel, default=1, const [1,322]}
       thumb_03_l_ctrl.Scale.Y {type=MovieSceneScriptingFloatChannel, default=1, const [1,322]}
       thumb_03_l_ctrl.Scale.Z {type=MovieSceneScriptingFloatChannel, default=1, const [1,322]}
       index_metacarpal_l_ctrl.Location.X {type=MovieSceneScriptingFloatChannel, default=0, const [0,322]}
       index_metacarpal_l_ctrl.Location.Y {type=MovieSceneScriptingFloatChannel, default=0, thinned_from=322, keys [[0,0],[354,0]]}
       index_metacarpal_l_ctrl.Location.Z {type=MovieSceneScriptingFloatChannel, default=0, thinned_from=322, keys [[0,0],[354,0]]}
       index_metacarpal_l_ctrl.Rotation.X {type=MovieSceneScriptingFloatChannel, default=0, const [-0.4323,322]}
       index_metacarpal_l_ctrl.Rotation.Y {type=MovieSceneScriptingFloatChannel, default=0, const [0.5781,322]}
       index_metacarpal_l_ctrl.Rotation.Z {type=MovieSceneScriptingFloatChannel, default=0, const [-0.9506,322]}
       index_metacarpal_l_ctrl.Scale.X {type=MovieSceneScriptingFloatChannel, default=1, const [1,322]}
       index_metacarpal_l_ctrl.Scale.Y {type=MovieSceneScriptingFloatChannel, default=1, const [1,322]}
       index_metacarpal_l_ctrl.Scale.Z {type=MovieSceneScriptingFloatChannel, default=1, const [1,322]}
       index_01_l_ctrl.Location.X {type=MovieSceneScriptingFloatChannel, default=0, thinned_from=322, keys [[0,0],[354,0]]}
       index_01_l_ctrl.Location.Y {type=MovieSceneScriptingFloatChannel, default=0, thinned_from=322, keys [[0,0],[354,0]]}
       index_01_l_ctrl.Location.Z {type=MovieSceneScriptingFloatChannel, default=0, thinned_from=322, keys [[0,0],[354,0]]}
       index_01_l_ctrl.Rotation.X {type=MovieSceneScriptingFloatChannel, default=0, const [3.8515,322]}
       index_01_l_ctrl.Rotation.Y {type=MovieSceneScriptingFloatChannel, default=0, const [-11.4472,322]}
       index_01_l_ctrl.Rotation.Z {type=MovieSceneScriptingFloatChannel, default=0, const [-4.3782,322]}
       index_01_l_ctrl.Scale.X {type=MovieSceneScriptingFloatChannel, default=1, const [1,322]}
       index_01_l_ctrl.Scale.Y {type=MovieSceneScriptingFloatChannel, default=1, const [1,322]}
       index_01_l_ctrl.Scale.Z {type=MovieSceneScriptingFloatChannel, default=1, const [1,322]}
       index_02_l_ctrl.Location.X {type=MovieSceneScriptingFloatChannel, default=0, const [0.3298,322]}
       index_02_l_ctrl.Location.Y {type=MovieSceneScriptingFloatChannel, default=0, const [-0.1751,322]}
       index_02_l_ctrl.Location.Z {type=MovieSceneScriptingFloatChannel, default=0, const [-0.0001,322]}
       index_02_l_ctrl.Rotation.X {type=MovieSceneScriptingFloatChannel, default=0, thinned_from=322, keys [[0,-0.3899],[86,-0.3888],[320,-0.3885],[354,-0.3899]]}
       index_02_l_ctrl.Rotation.Y {
        type=MovieSceneScriptingFloatChannel
        default=0
        thinned_from=322
        keys [[0,-0.035],[4,-0.035],[5,-0.0342],[10,-0.034],[11,-0.0333],[12,-0.0331],[15,-0.0331],[16,-0.0322],[17,-0.0324],[18,-0.0316],[20,-0.0315],[21,-0.0307],[22,-0.0305],[23,-0.0297],[25,-0.0296],[26,-0.0286],[27,-0.0286],[28,-0.0278],[29,-0.0277],[30,-0.0269],[31,-0.0267],[32,-0.0258],[33,-0.0255],[34,-0.0248],[35,-0.0247],[48,-0.0166],[49,-0.0165],[55,-0.012],[66,-0.0039],[67,-0.0036],[75,0.0028],[84,0.01],[85,0.0102],[93,0.0166],[96,0.0189],[97,0.0194],[98,0.0204],[107,0.0267],[108,0.0272],[122,0.0361],[123,0.0363],[129,0.0398],[130,0.04],[131,0.0407],[132,0.0409],[133,0.0416],[134,0.0417],[135,0.0424],[137,0.0426],[138,0.0433],[139,0.0435],[140,0.0442],[142,0.0444],[143,0.0451],[149,0.0463],[153,0.0468],[154,0.0468],[155,0.0471],[320,0.0471],[354,-0.035]]
       }
       index_02_l_ctrl.Rotation.Z {
        type=MovieSceneScriptingFloatChannel
        default=0
        thinned_from=322
        keys [[0,9.3395],[6,9.3395],[7,9.4664],[11,9.4664],[12,9.5934],[15,9.5934],[16,9.7202],[18,9.7202],[19,9.8561],[21,9.8561],[22,9.992],[23,9.992],[24,10.1278],[25,10.1278],[26,10.2636],[28,10.2636],[29,10.3993],[30,10.3993],[32,10.6707],[33,10.6707],[36,10.9459],[37,10.9459],[49,12.0454],[55,12.7078],[56,12.7078],[75,14.8939],[76,14.8939],[93,16.9184],[94,16.9184],[110,18.6504],[111,18.7032],[126,20.049],[127,20.049],[129,20.2234],[130,20.3501],[131,20.3502],[132,20.4769],[133,20.4769],[134,20.6036],[135,20.6036],[136,20.7303],[138,20.7303],[139,20.857],[140,20.857],[141,20.9836],[143,20.9836],[144,21.1102],[145,21.0636],[148,21.2035],[149,21.2035],[150,21.2502],[151,21.2502],[152,21.2968],[154,21.2968],[155,21.3434],[320,21.3434],[354,9.3395]]
       }
       index_02_l_ctrl.Scale.X {type=MovieSceneScriptingFloatChannel, default=1, const [1,322]}
       index_02_l_ctrl.Scale.Y {type=MovieSceneScriptingFloatChannel, default=1, const [1,322]}
       index_02_l_ctrl.Scale.Z {type=MovieSceneScriptingFloatChannel, default=1, const [1,322]}
       index_03_l_ctrl.Location.X {type=MovieSceneScriptingFloatChannel, default=0, thinned_from=322, keys [[0,0.2932],[354,0.2932]]}
       index_03_l_ctrl.Location.Y {type=MovieSceneScriptingFloatChannel, default=0, thinned_from=322, keys [[0,0.1198],[354,0.1198]]}
       index_03_l_ctrl.Location.Z {type=MovieSceneScriptingFloatChannel, default=0, thinned_from=322, keys [[0,0.005],[354,0.005]]}
       index_03_l_ctrl.Rotation.X {type=MovieSceneScriptingFloatChannel, default=0, thinned_from=322, keys [[0,-0.4305],[354,-0.4305]]}
       index_03_l_ctrl.Rotation.Y {type=MovieSceneScriptingFloatChannel, default=0, thinned_from=322, keys [[0,0.6661],[354,0.6661]]}
       index_03_l_ctrl.Rotation.Z {type=MovieSceneScriptingFloatChannel, default=0, thinned_from=322, keys [[0,0],[354,0]]}
       index_03_l_ctrl.Scale.X {type=MovieSceneScriptingFloatChannel, default=1, const [1,322]}
       index_03_l_ctrl.Scale.Y {type=MovieSceneScriptingFloatChannel, default=1, const [1,322]}
       index_03_l_ctrl.Scale.Z {type=MovieSceneScriptingFloatChannel, default=1, const [1,322]}
       middle_metacarpal_l_ctrl.Location.X {type=MovieSceneScriptingFloatChannel, default=0, const [0,322]}
       middle_metacarpal_l_ctrl.Location.Y {type=MovieSceneScriptingFloatChannel, default=0, thinned_from=322, keys [[0,0],[354,0]]}
       middle_metacarpal_l_ctrl.Location.Z {type=MovieSceneScriptingFloatChannel, default=0, thinned_from=322, keys [[0,0],[354,0]]}
       middle_metacarpal_l_ctrl.Rotation.X {type=MovieSceneScriptingFloatChannel, default=0, const [-0.9766,322]}
       middle_metacarpal_l_ctrl.Rotation.Y {type=MovieSceneScriptingFloatChannel, default=0, const [-4.3812,322]}
       middle_metacarpal_l_ctrl.Rotation.Z {type=MovieSceneScriptingFloatChannel, default=0, const [0.8157,322]}
       middle_metacarpal_l_ctrl.Scale.X {type=MovieSceneScriptingFloatChannel, default=1, const [1,322]}
       middle_metacarpal_l_ctrl.Scale.Y {type=MovieSceneScriptingFloatChannel, default=1, const [1,322]}
       middle_metacarpal_l_ctrl.Scale.Z {type=MovieSceneScriptingFloatChannel, default=1, const [1,322]}
       middle_01_l_ctrl.Location.X {type=MovieSceneScriptingFloatChannel, default=0, thinned_from=322, keys [[0,0],[354,0]]}
       middle_01_l_ctrl.Location.Y {type=MovieSceneScriptingFloatChannel, default=0, thinned_from=322, keys [[0,0],[354,0]]}
       middle_01_l_ctrl.Location.Z {type=MovieSceneScriptingFloatChannel, default=0, thinned_from=322, keys [[0,0],[354,0]]}
       middle_01_l_ctrl.Rotation.X {type=MovieSceneScriptingFloatChannel, default=0, const [2.3332,322]}
       middle_01_l_ctrl.Rotation.Y {type=MovieSceneScriptingFloatChannel, default=0, const [-5.0358,322]}
       middle_01_l_ctrl.Rotation.Z {type=MovieSceneScriptingFloatChannel, default=0, const [-10.7765,322]}
       middle_01_l_ctrl.Scale.X {type=MovieSceneScriptingFloatChannel, default=1, const [1,322]}
       middle_01_l_ctrl.Scale.Y {type=MovieSceneScriptingFloatChannel, default=1, const [1,322]}
       middle_01_l_ctrl.Scale.Z {type=MovieSceneScriptingFloatChannel, default=1, const [1,322]}
       middle_02_l_ctrl.Location.X {type=MovieSceneScriptingFloatChannel, default=0, const [0.0008,322]}
       middle_02_l_ctrl.Location.Y {type=MovieSceneScriptingFloatChannel, default=0, const [0.0009,322]}
       middle_02_l_ctrl.Location.Z {type=MovieSceneScriptingFloatChannel, default=0, const [0.0476,322]}
       middle_02_l_ctrl.Rotation.X {type=MovieSceneScriptingFloatChannel, default=0, thinned_from=322, keys [[0,-3.8567],[320,-3.8415],[354,-3.8567]]}
       middle_02_l_ctrl.Rotation.Y {
        type=MovieSceneScriptingFloatChannel
        default=0
        thinned_from=322
        keys [[0,-0.3333],[4,-0.3333],[5,-0.3289],[6,-0.3289],[7,-0.3246],[10,-0.3246],[12,-0.3158],[14,-0.3158],[16,-0.3071],[17,-0.3062],[18,-0.3007],[19,-0.3007],[20,-0.2966],[21,-0.2911],[22,-0.2911],[24,-0.2815],[25,-0.2775],[26,-0.272],[27,-0.272],[28,-0.2624],[29,-0.2624],[32,-0.2481],[33,-0.2391],[34,-0.2387],[41,-0.1955],[42,-0.1924],[48,-0.1554],[49,-0.1471],[50,-0.143],[51,-0.1361],[56,-0.0997],[57,-0.096],[64,-0.045],[65,-0.0357],[68,-0.0119],[69,-0.0075],[72,0.0163],[73,0.0239],[79,0.0715],[80,0.0756],[81,0.0852],[86,0.1246],[87,0.1285],[96,0.1993],[97,0.2041],[111,0.3014],[112,0.3066],[123,0.373],[124,0.3762],[126,0.388],[127,0.391],[128,0.397],[129,0.3985],[130,0.4072],[131,0.4074],[132,0.416],[133,0.4162],[134,0.4248],[135,0.425],[136,0.4337],[137,0.4339],[139,0.4425],[140,0.4427],[141,0.4513],[143,0.4515],[144,0.4601],[145,0.4586],[146,0.4651],[148,0.4653],[150,0.4719],[152,0.4719],[153,0.475],[156,0.475],[157,0.4786],[160,0.4786],[161,0.4771],[320,0.4771],[354,-0.3333]]
       }
       middle_02_l_ctrl.Rotation.Z {
        type=MovieSceneScriptingFloatChannel
        default=0
        thinned_from=322
        keys [[0,0.6939],[6,0.6937],[7,0.8213],[11,0.8212],[12,0.9487],[15,0.9485],[16,1.076],[17,1.0281],[18,1.1714],[20,1.1713],[21,1.3145],[23,1.3144],[24,1.4576],[25,1.4575],[26,1.6006],[27,1.6006],[28,1.7435],[30,1.7434],[31,1.8864],[32,1.8862],[33,2.0736],[34,2.0736],[41,2.7169],[42,2.7168],[48,3.2672],[49,3.4063],[56,4.1574],[57,4.1573],[64,4.9067],[65,5.0608],[79,6.6858],[80,6.6856],[81,6.8484],[96,8.5808],[97,8.6274],[112,10.1613],[113,10.2122],[126,11.38],[127,11.3799],[128,11.4696],[129,11.513],[130,11.6455],[131,11.6455],[132,11.7779],[133,11.7779],[134,11.9103],[135,11.9103],[136,12.0427],[138,12.0426],[139,12.175],[140,12.175],[141,12.3073],[143,12.3074],[144,12.4396],[145,12.3933],[146,12.5013],[149,12.5012],[150,12.6093],[156,12.6092],[157,12.7172],[160,12.7172],[161,12.6709],[320,12.6709],[354,0.6939]]
       }
       middle_02_l_ctrl.Scale.X {type=MovieSceneScriptingFloatChannel, default=1, const [1,322]}
       middle_02_l_ctrl.Scale.Y {type=MovieSceneScriptingFloatChannel, default=1, const [1,322]}
       middle_02_l_ctrl.Scale.Z {type=MovieSceneScriptingFloatChannel, default=1, const [1,322]}
       middle_03_l_ctrl.Location.X {type=MovieSceneScriptingFloatChannel, default=0, thinned_from=322, keys [[0,0],[354,0]]}
       middle_03_l_ctrl.Location.Y {type=MovieSceneScriptingFloatChannel, default=0, thinned_from=322, keys [[0,0],[354,0]]}
       middle_03_l_ctrl.Location.Z {type=MovieSceneScriptingFloatChannel, default=0, thinned_from=322, keys [[0,0],[354,0]]}
       middle_03_l_ctrl.Rotation.X {type=MovieSceneScriptingFloatChannel, default=0, thinned_from=322, keys [[0,0.744],[354,0.744]]}
       middle_03_l_ctrl.Rotation.Y {type=MovieSceneScriptingFloatChannel, default=0, const [2.2308,322]}
       middle_03_l_ctrl.Rotation.Z {type=MovieSceneScriptingFloatChannel, default=0, thinned_from=322, keys [[0,0],[354,0]]}
       middle_03_l_ctrl.Scale.X {type=MovieSceneScriptingFloatChannel, default=1, const [1,322]}
       middle_03_l_ctrl.Scale.Y {type=MovieSceneScriptingFloatChannel, default=1, const [1,322]}
       middle_03_l_ctrl.Scale.Z {type=MovieSceneScriptingFloatChannel, default=1, const [1,322]}
       ring_metacarpal_l_ctrl.Location.X {type=MovieSceneScriptingFloatChannel, default=0, thinned_from=322, keys [[0,0],[354,0]]}
       ring_metacarpal_l_ctrl.Location.Y {type=MovieSceneScriptingFloatChannel, default=0, thinned_from=322, keys [[0,0],[354,0]]}
       ring_metacarpal_l_ctrl.Location.Z {type=MovieSceneScriptingFloatChannel, default=0, thinned_from=322, keys [[0,0],[354,0]]}
       ring_metacarpal_l_ctrl.Rotation.X {type=MovieSceneScriptingFloatChannel, default=0, const [-2.6065,322]}
       ring_metacarpal_l_ctrl.Rotation.Y {type=MovieSceneScriptingFloatChannel, default=0, const [-5.9585,322]}
       ring_metacarpal_l_ctrl.Rotation.Z {type=MovieSceneScriptingFloatChannel, default=0, const [9.7527,322]}
       ring_metacarpal_l_ctrl.Scale.X {type=MovieSceneScriptingFloatChannel, default=1, const [1,322]}
       ring_metacarpal_l_ctrl.Scale.Y {type=MovieSceneScriptingFloatChannel, default=1, const [1,322]}
       ring_metacarpal_l_ctrl.Scale.Z {type=MovieSceneScriptingFloatChannel, default=1, const [1,322]}
       ring_01_l_ctrl.Location.X {type=MovieSceneScriptingFloatChannel, default=0, thinned_from=322, keys [[0,0],[354,0]]}
       ring_01_l_ctrl.Location.Y {type=MovieSceneScriptingFloatChannel, default=0, thinned_from=322, keys [[0,0],[354,0]]}
       ring_01_l_ctrl.Location.Z {type=MovieSceneScriptingFloatChannel, default=0, thinned_from=322, keys [[0,0],[354,0]]}
       ring_01_l_ctrl.Rotation.X {
        type=MovieSceneScriptingFloatChannel
        default=0
        thinned_from=322
        keys [[0,1.5312],[2,1.5312],[3,1.5416],[4,1.5416],[5,1.521],[6,1.5205],[7,1.531],[8,1.531],[9,1.5208],[10,1.5204],[11,1.5102],[12,1.5098],[13,1.4997],[14,1.4792],[15,1.4892],[16,1.4788],[17,1.4832],[18,1.4578],[19,1.4444],[20,1.4557],[24,1.4016],[25,1.4013],[28,1.3611],[29,1.3594],[32,1.3198],[33,1.291],[34,1.2889],[37,1.2337],[38,1.2318],[39,1.2064],[42,1.1523],[43,1.1417],[48,1.0543],[49,1.0074],[50,1.023],[61,0.7736],[64,0.7087],[65,0.6832],[66,0.6762],[68,0.6209],[69,0.5998],[75,0.4409],[76,0.4138],[77,0.4079],[78,0.3762],[81,0.2999],[82,0.2713],[83,0.2465],[85,0.1899],[86,0.1656],[88,0.1098],[90,0.0546],[91,0.0272],[92,-0.0054],[93,-0.0326],[94,-0.0633],[95,-0.0903],[96,-0.1207],[97,-0.1466],[98,-0.1736],[101,-0.2565],[102,-0.2847],[104,-0.3424],[105,-0.3718],[106,-0.4031],[109,-0.4937],[110,-0.525],[111,-0.5584],[112,-0.5895],[113,-0.6191],[114,-0.645],[115,-0.6776],[116,-0.7033],[117,-0.7397],[118,-0.7653],[124,-0.9583],[125,-0.9609],[126,-0.997],[127,-1.0357],[128,-1.0674],[129,-1.0901],[132,-1.1849],[133,-1.2019],[135,-1.2649],[136,-1.2995],[140,-1.4251],[141,-1.4596],[142,-1.471],[143,-1.5023],[144,-1.5514],[145,-1.5667],[146,-1.5982],[147,-1.6222],[151,-1.7432],[153,-1.7918],[154,-1.8246],[158,-1.9348],[159,-2.8389],[320,-2.8393],[354,1.5312]]
       }
       ring_01_l_ctrl.Rotation.Y {
        type=MovieSceneScriptingFloatChannel
        default=0
        thinned_from=322
        keys [[0,-6.7802],[4,-6.7764],[5,-6.7204],[8,-6.7182],[9,-6.6584],[10,-6.66],[11,-6.6002],[12,-6.6018],[14,-6.486],[15,-6.4838],[16,-6.4256],[17,-6.4018],[19,-6.2618],[20,-6.2591],[24,-5.9721],[25,-5.9736],[32,-5.4696],[33,-5.3565],[34,-5.3065],[37,-5.0206],[38,-4.9715],[48,-4.0211],[49,-3.8613],[50,-3.8585],[64,-2.2561],[65,-2.1327],[66,-2.0707],[76,-0.8413],[77,-0.7792],[78,-0.6585],[80,-0.4131],[81,-0.2921],[82,-0.174],[83,-0.0547],[85,0.1816],[86,0.301],[96,1.482],[97,1.5853],[112,3.1951],[124,4.3123],[125,4.3088],[128,4.5862],[141,5.3887],[142,5.3835],[144,5.5025],[145,5.5184],[148,5.6166],[151,5.6893],[156,5.7531],[158,5.7597],[159,5.3727],[320,5.3728],[354,-6.7802]]
       }
       ring_01_l_ctrl.Rotation.Z {
        type=MovieSceneScriptingFloatChannel
        default=0
        thinned_from=322
        keys [[0,-18.382],[5,-18.3761],[6,-18.3396],[9,-18.3345],[16,-18.1653],[17,-18.1638],[23,-17.9585],[32,-17.633],[33,-17.5668],[47,-16.8719],[48,-16.7937],[49,-16.7688],[55,-16.3602],[56,-16.3511],[66,-15.6672],[67,-15.6268],[75,-15.0563],[91,-13.9732],[92,-13.9356],[105,-13.0775],[106,-13.0429],[110,-12.7849],[111,-12.7507],[116,-12.4385],[117,-12.4355],[125,-11.9952],[126,-11.9929],[129,-11.829],[135,-11.6065],[136,-11.6056],[140,-11.4586],[141,-11.4579],[143,-11.3857],[144,-11.3855],[145,-11.3502],[146,-11.3501],[147,-11.2779],[151,-11.2787],[152,-11.2068],[158,-11.2099],[159,-20.5083],[320,-20.5083],[354,-18.382]]
       }
       ring_01_l_ctrl.Scale.X {type=MovieSceneScriptingFloatChannel, default=1, const [1,322]}
       ring_01_l_ctrl.Scale.Y {type=MovieSceneScriptingFloatChannel, default=1, const [1,322]}
       ring_01_l_ctrl.Scale.Z {type=MovieSceneScriptingFloatChannel, default=1, const [1,322]}
       ring_02_l_ctrl.Location.X {type=MovieSceneScriptingFloatChannel, default=0, thinned_from=322, keys [[0,0],[354,0]]}
       ring_02_l_ctrl.Location.Y {type=MovieSceneScriptingFloatChannel, default=0, thinned_from=322, keys [[0,0],[354,0]]}
       ring_02_l_ctrl.Location.Z {type=MovieSceneScriptingFloatChannel, default=0, thinned_from=322, keys [[0,0],[354,0]]}
       ring_02_l_ctrl.Rotation.X {
        type=MovieSceneScriptingFloatChannel
        default=0
        thinned_from=322
        keys [[0,-2.4658],[10,-2.4568],[14,-2.4434],[29,-2.3814],[46,-2.2695],[49,-2.2429],[50,-2.2389],[80,-1.9734],[81,-1.9621],[93,-1.8527],[121,-1.628],[138,-1.5342],[144,-1.5119],[153,-1.4925],[154,-1.4925],[155,-1.489],[320,-1.489],[354,-2.4658]]
       }
       ring_02_l_ctrl.Rotation.Y {
        type=MovieSceneScriptingFloatChannel
        default=0
        thinned_from=322
        keys [[0,2.1623],[10,2.1691],[34,2.2505],[48,2.3272],[127,2.8433],[320,2.922],[354,2.1623]]
       }
       ring_02_l_ctrl.Rotation.Z {
        type=MovieSceneScriptingFloatChannel
        default=0
        thinned_from=322
        keys [[0,-7.2883],[4,-7.2883],[5,-7.188],[7,-7.188],[8,-7.0878],[10,-7.0877],[12,-6.8873],[13,-6.8873],[14,-6.7871],[15,-6.7871],[16,-6.687],[17,-6.5519],[18,-6.5518],[20,-6.3068],[21,-6.3067],[32,-4.9619],[33,-4.8455],[34,-4.764],[38,-4.1128],[39,-4.0315],[48,-2.5709],[49,-2.3113],[50,-2.2137],[57,-0.8493],[58,-0.752],[62,0.0253],[64,0.4134],[65,0.659],[68,1.2798],[69,1.4349],[80,3.7026],[81,3.9612],[82,4.0656],[96,6.9794],[97,7.1302],[99,7.5061],[100,7.7409],[106,8.8666],[107,9.0072],[111,9.7562],[112,9.8966],[113,10.1156],[122,11.531],[123,11.6096],[127,12.2375],[128,12.316],[129,12.5008],[138,13.506],[139,13.5061],[142,13.8408],[143,13.8408],[144,13.9524],[145,13.9802],[148,14.2312],[149,14.2312],[150,14.3148],[151,14.3148],[152,14.3984],[154,14.3984],[155,14.482],[320,14.482],[354,-7.2883]]
       }
       ring_02_l_ctrl.Scale.X {type=MovieSceneScriptingFloatChannel, default=1, const [1,322]}
       ring_02_l_ctrl.Scale.Y {type=MovieSceneScriptingFloatChannel, default=1, const [1,322]}
       ring_02_l_ctrl.Scale.Z {type=MovieSceneScriptingFloatChannel, default=1, const [1,322]}
       ring_03_l_ctrl.Location.X {type=MovieSceneScriptingFloatChannel, default=0, const [0,322]}
       ring_03_l_ctrl.Location.Y {type=MovieSceneScriptingFloatChannel, default=0, const [0,322]}
       ring_03_l_ctrl.Location.Z {type=MovieSceneScriptingFloatChannel, default=0, const [0,322]}
       ring_03_l_ctrl.Rotation.X {type=MovieSceneScriptingFloatChannel, default=0, const [0,322]}
       ring_03_l_ctrl.Rotation.Y {type=MovieSceneScriptingFloatChannel, default=0, const [0,322]}
       ring_03_l_ctrl.Rotation.Z {type=MovieSceneScriptingFloatChannel, default=0, const [0,322]}
       ring_03_l_ctrl.Scale.X {type=MovieSceneScriptingFloatChannel, default=1, const [1,322]}
       ring_03_l_ctrl.Scale.Y {type=MovieSceneScriptingFloatChannel, default=1, const [1,322]}
       ring_03_l_ctrl.Scale.Z {type=MovieSceneScriptingFloatChannel, default=1, const [1,322]}
       pinky_metacarpal_l_ctrl.Location.X {type=MovieSceneScriptingFloatChannel, default=0, thinned_from=322, keys [[0,0],[354,0]]}
       pinky_metacarpal_l_ctrl.Location.Y {type=MovieSceneScriptingFloatChannel, default=0, thinned_from=322, keys [[0,0],[354,0]]}
       pinky_metacarpal_l_ctrl.Location.Z {type=MovieSceneScriptingFloatChannel, default=0, thinned_from=322, keys [[0,0],[354,0]]}
       pinky_metacarpal_l_ctrl.Rotation.X {type=MovieSceneScriptingFloatChannel, default=0, const [-7.089,322]}
       pinky_metacarpal_l_ctrl.Rotation.Y {type=MovieSceneScriptingFloatChannel, default=0, const [-6.2626,322]}
       pinky_metacarpal_l_ctrl.Rotation.Z {type=MovieSceneScriptingFloatChannel, default=0, const [2.9135,322]}
       pinky_metacarpal_l_ctrl.Scale.X {type=MovieSceneScriptingFloatChannel, default=1, const [1,322]}
       pinky_metacarpal_l_ctrl.Scale.Y {type=MovieSceneScriptingFloatChannel, default=1, const [1,322]}
       pinky_metacarpal_l_ctrl.Scale.Z {type=MovieSceneScriptingFloatChannel, default=1, const [1,322]}
       pinky_01_l_ctrl.Location.X {type=MovieSceneScriptingFloatChannel, default=0, thinned_from=322, keys [[0,0],[354,0]]}
       pinky_01_l_ctrl.Location.Y {type=MovieSceneScriptingFloatChannel, default=0, thinned_from=322, keys [[0,0],[354,0]]}
       pinky_01_l_ctrl.Location.Z {type=MovieSceneScriptingFloatChannel, default=0, thinned_from=322, keys [[0,0],[354,0]]}
       pinky_01_l_ctrl.Rotation.X {type=MovieSceneScriptingFloatChannel, default=0, const [-3.7473,322]}
       pinky_01_l_ctrl.Rotation.Y {type=MovieSceneScriptingFloatChannel, default=0, const [-13.136,322]}
       pinky_01_l_ctrl.Rotation.Z {type=MovieSceneScriptingFloatChannel, default=0, const [-0.1797,322]}
       pinky_01_l_ctrl.Scale.X {type=MovieSceneScriptingFloatChannel, default=1, const [1,322]}
       pinky_01_l_ctrl.Scale.Y {type=MovieSceneScriptingFloatChannel, default=1, const [1,322]}
       pinky_01_l_ctrl.Scale.Z {type=MovieSceneScriptingFloatChannel, default=1, const [1,322]}
       pinky_02_l_ctrl.Location.X {type=MovieSceneScriptingFloatChannel, default=0, thinned_from=322, keys [[0,0],[354,0]]}
       pinky_02_l_ctrl.Location.Y {type=MovieSceneScriptingFloatChannel, default=0, thinned_from=322, keys [[0,0],[354,0]]}
       pinky_02_l_ctrl.Location.Z {type=MovieSceneScriptingFloatChannel, default=0, thinned_from=322, keys [[0,0],[354,0]]}
       pinky_02_l_ctrl.Rotation.X {type=MovieSceneScriptingFloatChannel, default=0, const [0.9973,322]}
       pinky_02_l_ctrl.Rotation.Y {type=MovieSceneScriptingFloatChannel, default=0, const [3.5352,322]}
       pinky_02_l_ctrl.Rotation.Z {type=MovieSceneScriptingFloatChannel, default=0, const [16.3636,322]}
       pinky_02_l_ctrl.Scale.X {type=MovieSceneScriptingFloatChannel, default=1, const [1,322]}
       pinky_02_l_ctrl.Scale.Y {type=MovieSceneScriptingFloatChannel, default=1, const [1,322]}
       pinky_02_l_ctrl.Scale.Z {type=MovieSceneScriptingFloatChannel, default=1, const [1,322]}
       pinky_03_l_ctrl.Location.X {type=MovieSceneScriptingFloatChannel, default=0, thinned_from=322, keys [[0,0],[354,0]]}
       pinky_03_l_ctrl.Location.Y {type=MovieSceneScriptingFloatChannel, default=0, thinned_from=322, keys [[0,0],[354,0]]}
       pinky_03_l_ctrl.Location.Z {type=MovieSceneScriptingFloatChannel, default=0, thinned_from=322, keys [[0,0],[354,0]]}
       pinky_03_l_ctrl.Rotation.X {type=MovieSceneScriptingFloatChannel, default=0, const [-0.1956,322]}
       pinky_03_l_ctrl.Rotation.Y {type=MovieSceneScriptingFloatChannel, default=0, const [1.1092,322]}
       pinky_03_l_ctrl.Rotation.Z {type=MovieSceneScriptingFloatChannel, default=0, const [0,322]}
       pinky_03_l_ctrl.Scale.X {type=MovieSceneScriptingFloatChannel, default=1, const [1,322]}
       pinky_03_l_ctrl.Scale.Y {type=MovieSceneScriptingFloatChannel, default=1, const [1,322]}
       pinky_03_l_ctrl.Scale.Z {type=MovieSceneScriptingFloatChannel, default=1, const [1,322]}
       hand_l_ik_ctrl.Location.X {
        type=MovieSceneScriptingFloatChannel
        default=0
        thinned_from=322
        keys [[0,-60.3991],[1,-60.8728],[5,-61.1541],[11,-60.5818],[27,-57.9081],[32,-57.6101],[64,-59.6689],[73,-59.2279],[96,-55.68],[102,-55.4546],[109,-56.0921],[119,-58.3358],[133,-62.3553],[140,-63.3048],[222,-58.4689],[233,-57.1164],[240,-57.0532],[281,-62.0689],[295,-62.3108],[306,-60.9725],[320,-58.2856],[354,-60.3991]]
       }
       hand_l_ik_ctrl.Location.Y {
        type=MovieSceneScriptingFloatChannel
        default=0
        thinned_from=322
        keys [[0,-1.8073],[1,-2.2119],[2,-2.3238],[3,-2.425],[4,-2.5117],[5,-2.5765],[6,-2.6252],[7,-2.6553],[8,-2.6721],[9,-2.6737],[10,-2.6665],[12,-2.6366],[14,-2.5895],[16,-2.5276],[17,-2.5065],[18,-2.4766],[22,-2.4066],[24,-2.3592],[25,-2.3183],[26,-2.2853],[27,-2.2357],[29,-2.1591],[31,-2.0751],[32,-2.0473],[33,-2.0243],[34,-2.0121],[35,-2.0053],[37,-2.0201],[39,-2.0594],[40,-2.0951],[41,-2.1351],[43,-2.2263],[45,-2.3354],[47,-2.4647],[49,-2.6077],[51,-2.7638],[52,-2.8372],[63,-3.729],[64,-3.7976],[66,-3.9434],[68,-4.067],[69,-4.116],[71,-4.1873],[73,-4.233],[74,-4.2442],[76,-4.2444],[83,-4.1619],[84,-4.1447],[86,-4.0691],[87,-4.0212],[88,-3.9638],[90,-3.8287],[91,-3.7818],[92,-3.75],[94,-3.7342],[95,-3.7555],[97,-3.8208],[98,-3.8604],[99,-3.9154],[100,-3.9613],[104,-4.2305],[106,-4.3318],[108,-4.4152],[109,-4.4717],[110,-4.5081],[112,-4.6031],[114,-4.6739],[119,-4.8631],[121,-4.9501],[124,-5.1048],[126,-5.2313],[128,-5.3307],[130,-5.4012],[134,-5.4571],[136,-5.4571],[137,-5.4457],[139,-5.3942],[141,-5.3184],[143,-5.2075],[146,-4.997],[154,-4.3779],[158,-4.0997],[160,-3.9747],[164,-3.7736],[166,-3.6825],[170,-3.5231],[172,-3.4655],[174,-3.4335],[175,-3.4295],[176,-3.4341],[177,-3.4487],[179,-3.5116],[180,-3.5531],[182,-3.6564],[185,-3.7809],[194,-4.0838],[195,-4.1249],[197,-4.197],[200,-4.3239],[202,-4.3948],[209,-4.7203],[212,-4.8382],[214,-4.8937],[215,-4.9096],[216,-4.8931],[217,-4.8572],[218,-4.8034],[219,-4.7238],[220,-4.6047],[221,-4.4453],[222,-4.2641],[223,-4.0362],[224,-3.7662],[225,-3.4622],[226,-3.1113],[227,-2.7221],[228,-2.2938],[229,-1.829],[230,-1.3349],[231,-0.805],[232,-0.2547],[233,0.3213],[234,0.9151],[235,1.5386],[236,2.1791],[237,2.8419],[238,3.5191],[241,5.5919],[242,6.2726],[243,6.9398],[244,7.5836],[245,8.1973],[246,8.7829],[247,9.3343],[248,9.8468],[249,10.3194],[250,10.747],[251,11.134],[252,11.463],[253,11.749],[254,11.984],[256,12.3725],[258,12.6701],[260,12.8913],[262,13.0576],[266,13.2597],[272,13.3165],[293,12.952],[296,12.7955],[299,12.5407],[301,12.2998],[310,10.9498],[316,10.1563],[320,9.6909],[354,-1.8073]]
       }
       hand_l_ik_ctrl.Location.Z {
        type=MovieSceneScriptingFloatChannel
        default=0
        thinned_from=322
        keys [[0,46.1366],[22,42.9627],[30,42.3762],[69,44.3124],[86,43.3292],[95,41.9652],[104,40.1776],[109,39.6224],[114,39.5782],[122,40.345],[131,41.4695],[135,41.6591],[140,41.4752],[182,36.2425],[196,35.8629],[205,36.1974],[214,37.2159],[218,37.9923],[221,38.8691],[229,41.9776],[236,44.9605],[240,46.284],[243,46.9624],[247,47.4456],[320,47.8977],[354,46.1366]]
       }
       hand_l_ik_ctrl.Rotation.X {
        type=MovieSceneScriptingFloatChannel
        default=0
        thinned_from=322
        keys [[0,138.6421],[1,139.2649],[58,144.5071],[119,140.305],[141,143.9878],[209,144.9181],[220,141.9747],[227,138.0455],[245,122.4032],[251,118.4893],[259,115.6117],[297,114.2225],[320,118.3485],[354,138.6421]]
       }
       hand_l_ik_ctrl.Rotation.Y {
        type=MovieSceneScriptingFloatChannel
        default=0
        thinned_from=322
        keys [[0,13.2302],[1,13.5325],[3,13.7681],[6,14.2163],[11,15.1206],[13,15.4148],[25,16.645],[27,16.7992],[29,16.8802],[32,16.8566],[51,16.588],[77,17.3237],[80,17.4959],[86,18.0626],[87,18.2206],[88,18.4384],[89,18.7124],[90,19.0721],[91,19.5351],[92,20.0895],[93,20.7204],[94,21.4215],[95,22.1797],[97,23.826],[103,29.109],[105,30.7459],[107,32.1546],[109,33.3745],[111,34.3848],[113,35.1832],[115,35.7672],[117,36.1968],[120,36.5493],[125,36.5794],[135,35.7697],[145,35.8095],[170,37.5204],[176,38.3616],[184,40.1492],[193,41.0599],[204,41.2819],[213,40.6442],[215,40.3355],[217,39.8496],[219,39.1563],[221,38.1646],[223,36.9847],[225,35.6176],[227,34.0165],[229,32.1654],[231,30.0554],[232,28.9056],[233,27.6913],[235,25.1088],[237,22.3649],[239,19.5054],[242,15.0702],[244,12.075],[245,10.6037],[246,9.172],[247,7.7944],[248,6.4792],[249,5.2253],[250,4.0358],[251,2.9164],[252,1.8674],[253,0.897],[254,0.0086],[255,-0.8314],[256,-1.6245],[257,-2.3885],[258,-3.1278],[259,-3.8407],[260,-4.5291],[261,-5.1949],[262,-5.8328],[263,-6.4327],[264,-6.9971],[265,-7.5156],[266,-7.9939],[267,-8.4304],[268,-8.8234],[269,-9.1741],[271,-9.7963],[274,-10.6233],[276,-11.1029],[278,-11.5227],[280,-11.8915],[282,-12.2099],[284,-12.478],[286,-12.7078],[290,-13.0648],[292,-13.1923],[293,-13.2249],[294,-13.2258],[295,-13.1937],[297,-13.0421],[298,-12.9204],[299,-12.7658],[300,-12.5811],[302,-12.131],[304,-11.5932],[306,-10.9998],[310,-9.7333],[312,-9.137],[316,-8.036],[318,-7.5338],[320,-7.0753],[354,13.2302]]
       }
       hand_l_ik_ctrl.Rotation.Z {
        type=MovieSceneScriptingFloatChannel
        default=0
        thinned_from=322
        keys [[0,134.7565],[27,129.9734],[92,128.8242],[116,126.6987],[218,129.6547],[261,144.8502],[295,147.1182],[320,141.62],[354,134.7565]]
       }
       hand_l_ik_ctrl.Scale.X {type=MovieSceneScriptingFloatChannel, default=1, const [1,322]}
       hand_l_ik_ctrl.Scale.Y {type=MovieSceneScriptingFloatChannel, default=1, const [1,322]}
       hand_l_ik_ctrl.Scale.Z {type=MovieSceneScriptingFloatChannel, default=1, const [1,322]}
       Softness {type=MovieSceneScriptingFloatChannel, default=0, const [0,322]}
       hand_l_ikorient_ctrl.Rotation.Roll {type=MovieSceneScriptingFloatChannel, default=0, const [0,322]}
       hand_l_ikorient_ctrl.Rotation.Pitch {type=MovieSceneScriptingFloatChannel, default=0, const [0,322]}
       hand_l_ikorient_ctrl.Rotation.Yaw {type=MovieSceneScriptingFloatChannel, default=0, const [0,322]}
       arm_l_pv_ik_ctrl.Location.X {
        type=MovieSceneScriptingFloatChannel
        default=0
        thinned_from=322
        keys [[0,12.918],[1,13.0569],[5,13.2285],[13,13.3549],[21,13.6232],[24,13.6504],[40,13.3223],[48,13.4019],[52,13.5329],[76,14.8527],[84,15.0699],[86,15.0756],[88,14.9943],[89,14.9085],[91,14.6442],[93,14.2715],[95,13.8],[96,13.5344],[99,12.6137],[102,11.5618],[106,10.0959],[108,9.4058],[110,8.7792],[111,8.4935],[113,7.981],[114,7.7571],[116,7.3649],[117,7.181],[119,6.862],[120,6.7184],[121,6.5893],[124,6.2899],[126,6.1353],[128,6.0093],[132,5.8083],[139,5.4031],[154,4.6393],[165,4.1121],[169,3.8858],[173,3.6377],[176,3.4189],[178,3.2538],[180,3.0673],[181,2.9852],[183,2.8358],[185,2.7258],[187,2.656],[189,2.6152],[191,2.5895],[193,2.5888],[196,2.6315],[197,2.6562],[199,2.7192],[201,2.7958],[202,2.8385],[204,2.9511],[206,3.0856],[207,3.1577],[209,3.3227],[211,3.5117],[213,3.7246],[214,3.8457],[216,4.1218],[217,4.2703],[218,4.4278],[219,4.5993],[222,5.1476],[225,5.6394],[230,6.3545],[234,6.9134],[235,7.0367],[236,7.1441],[237,7.2251],[238,7.2821],[239,7.3078],[240,7.3136],[241,7.2835],[242,7.2348],[243,7.1556],[244,7.0577],[246,6.8087],[251,6.0544],[253,5.7915],[257,5.3587],[259,5.1584],[262,4.9081],[266,4.6145],[269,4.4412],[272,4.2953],[273,4.2622],[275,4.1719],[278,4.0616],[281,3.9951],[282,3.9887],[283,3.9707],[286,3.9554],[287,3.9612],[288,3.9556],[289,3.9725],[290,3.98],[292,4.0196],[293,4.0576],[295,4.2023],[296,4.3051],[297,4.423],[298,4.5636],[299,4.7187],[301,5.0941],[304,5.7503],[310,7.1445],[314,7.9916],[318,8.7853],[320,9.1471],[354,12.918]]
       }
       arm_l_pv_ik_ctrl.Location.Y {
        type=MovieSceneScriptingFloatChannel
        default=0
        thinned_from=322
        keys [[0,28.9734],[1,28.3882],[10,26.1896],[13,25.6194],[21,24.6494],[27,24.2274],[30,24.1796],[61,25.0038],[79,24.7657],[86,24.2686],[87,24.1258],[88,23.9304],[89,23.6831],[90,23.3659],[91,22.9644],[92,22.4986],[93,21.9612],[94,21.376],[96,20.0785],[103,15.0523],[104,14.3727],[105,13.7412],[106,13.1544],[107,12.6283],[108,12.1337],[109,11.685],[110,11.2905],[111,10.9295],[112,10.608],[113,10.33],[114,10.0928],[115,9.8936],[116,9.7227],[117,9.589],[118,9.4983],[119,9.4319],[120,9.3901],[121,9.3769],[122,9.3834],[124,9.4422],[126,9.5323],[128,9.6597],[131,9.8939],[132,9.9539],[134,10.0081],[135,10.0152],[137,9.9553],[138,9.9011],[140,9.7385],[142,9.5197],[144,9.271],[155,7.6791],[159,7.1571],[162,6.809],[173,5.6743],[176,5.3201],[177,5.1889],[181,4.5897],[182,4.455],[183,4.3359],[184,4.245],[185,4.1626],[186,4.113],[188,4.0446],[190,4.0016],[192,3.9868],[194,3.9911],[196,4.0097],[198,4.0558],[200,4.1208],[202,4.214],[204,4.3395],[205,4.4128],[207,4.5892],[209,4.8027],[211,5.0516],[213,5.3437],[214,5.5169],[215,5.7175],[216,5.9565],[217,6.2414],[218,6.5771],[219,6.9787],[220,7.466],[221,8.0205],[222,8.6291],[223,9.2731],[224,9.9616],[225,10.6947],[226,11.4892],[227,12.3305],[228,13.221],[229,14.1527],[230,15.147],[231,16.1741],[232,17.2494],[233,18.3726],[235,20.72],[244,31.8865],[246,34.2198],[248,36.3488],[250,38.2791],[252,39.9721],[254,41.4209],[258,43.881],[262,45.969],[265,47.253],[268,48.2412],[279,50.4958],[292,51.6495],[297,51.5159],[302,50.8081],[320,46.9059],[354,28.9734]]
       }
       arm_l_pv_ik_ctrl.Location.Z {
        type=MovieSceneScriptingFloatChannel
        default=0
        thinned_from=322
        keys [[0,37.3524],[1,37.7692],[4,37.9498],[7,37.7967],[28,34.2147],[31,33.9129],[35,33.9061],[41,34.3573],[64,36.9945],[69,37.2364],[74,36.9799],[94,33.9194],[98,33.568],[102,33.5566],[107,34.01],[112,34.8884],[126,38.3475],[130,39.1109],[136,39.6843],[141,39.7601],[184,34.3471],[197,33.5432],[204,33.5729],[214,34.429],[219,34.9125],[223,34.975],[232,34.5572],[246,35.0117],[260,36.2471],[280,38.4907],[295,39.1508],[306,38.7899],[320,37.5082],[354,37.3524]]
       }
       arm_l_pv_ik_ctrl.Rotation.X {
        type=MovieSceneScriptingFloatChannel
        default=0
        thinned_from=322
        keys [[0,-21.2312],[1,-21.4583],[7,-21.7297],[19,-21.7819],[41,-22.9933],[52,-23.1458],[62,-22.7088],[79,-21.6717],[84,-21.5098],[87,-21.2789],[90,-20.8441],[97,-19.5094],[106,-17.4893],[108,-17.1097],[111,-16.6854],[114,-16.369],[116,-16.2365],[118,-16.2221],[120,-16.3279],[122,-16.5314],[124,-16.8213],[133,-18.5764],[135,-18.8895],[137,-19.1142],[139,-19.241],[144,-19.2633],[159,-18.68],[183,-18.7941],[195,-19.061],[204,-18.9764],[209,-18.7543],[212,-18.5044],[221,-17.3676],[231,-16.0664],[234,-15.5751],[240,-14.4089],[242,-14.1023],[250,-13.2352],[253,-13.0158],[255,-12.9405],[268,-12.901],[275,-13.0592],[281,-13.2746],[288,-13.3276],[294,-13.2132],[320,-12.0548],[354,-21.2312]]
       }
       arm_l_pv_ik_ctrl.Rotation.Y {
        type=MovieSceneScriptingFloatChannel
        default=0
        thinned_from=322
        keys [[0,12.9791],[1,13.4401],[3,13.8279],[6,14.3043],[11,14.9037],[14,15.4199],[20,16.6994],[23,17.2351],[25,17.501],[31,17.9661],[42,18.296],[54,18.3256],[84,17.3671],[86,17.2501],[88,17.0381],[90,16.7263],[93,16.0888],[96,15.2976],[99,14.3809],[107,11.6913],[109,11.0727],[111,10.5338],[112,10.2933],[113,10.0796],[115,9.7184],[117,9.4355],[118,9.3231],[119,9.2379],[120,9.1773],[121,9.1483],[122,9.1481],[123,9.173],[125,9.2784],[130,9.7179],[132,9.8495],[135,9.9592],[142,10.0176],[156,9.8226],[164,9.5909],[167,9.4482],[179,8.6164],[183,8.4197],[191,8.1799],[205,7.9613],[208,7.8655],[211,7.7063],[213,7.5535],[215,7.3597],[216,7.2324],[217,7.0766],[218,6.8986],[219,6.6938],[220,6.4576],[221,6.1882],[222,5.8841],[223,5.5404],[224,5.1618],[226,4.3503],[228,3.529],[229,3.1344],[230,2.7647],[231,2.4234],[232,2.1259],[233,1.8814],[234,1.685],[235,1.5208],[236,1.3827],[237,1.2598],[238,1.155],[239,1.0623],[240,1.0011],[241,0.9781],[242,0.9986],[243,1.0657],[244,1.1721],[245,1.3022],[247,1.5783],[249,1.8509],[251,2.1415],[255,2.7787],[256,2.927],[257,3.0869],[258,3.2582],[259,3.4427],[260,3.6422],[263,4.2837],[264,4.4873],[265,4.6701],[266,4.8428],[267,5.0018],[268,5.1495],[269,5.2813],[277,6.2398],[282,6.7697],[285,7.0402],[287,7.1905],[289,7.3195],[292,7.4626],[295,7.5584],[310,7.7802],[313,7.8682],[316,8.0138],[320,8.2851],[354,12.9791]]
       }
       arm_l_pv_ik_ctrl.Rotation.Z {
        type=MovieSceneScriptingFloatChannel
        default=0
        thinned_from=322
        keys [[0,45.3929],[1,44.6682],[22,39.0246],[27,38.0407],[30,37.7316],[79,37.6587],[87,37.0987],[89,36.7864],[90,36.5289],[91,36.1723],[92,35.7253],[94,34.6064],[96,33.2631],[103,28.0541],[105,26.7148],[107,25.6111],[109,24.6632],[111,23.8851],[113,23.2856],[115,22.8663],[117,22.5474],[121,22.1573],[127,21.8038],[150,21.3959],[169,20.3074],[173,19.9213],[176,19.4684],[178,19.0424],[182,18.0135],[184,17.5747],[185,17.3915],[189,16.8978],[193,16.5544],[198,16.3365],[201,16.3139],[204,16.4046],[207,16.6295],[209,16.8665],[211,17.1837],[213,17.5885],[214,17.8375],[215,18.1419],[216,18.51],[217,18.9505],[218,19.4497],[219,20.0383],[220,20.7342],[221,21.5215],[222,22.3758],[223,23.3027],[224,24.3133],[225,25.3948],[226,26.5577],[227,27.7966],[228,29.1109],[229,30.499],[231,33.48],[233,36.693],[236,41.8425],[244,56.3625],[247,61.3885],[249,64.4048],[251,67.0933],[253,69.4068],[257,73.1117],[261,76.1176],[265,78.4705],[269,80.1431],[280,82.6816],[293,83.6541],[298,83.1441],[306,80.4908],[320,75.0141],[354,45.3929]]
       }
       arm_l_pv_ik_ctrl.Scale.X {type=MovieSceneScriptingFloatChannel, default=1, const [1,322]}
       arm_l_pv_ik_ctrl.Scale.Y {type=MovieSceneScriptingFloatChannel, default=1, const [1,322]}
       arm_l_pv_ik_ctrl.Scale.Z {type=MovieSceneScriptingFloatChannel, default=1, const [1,322]}
       upperarm_r_fk_ctrl.Location.X {type=MovieSceneScriptingFloatChannel, default=0, thinned_from=322, keys [[0,0],[354,0]]}
       upperarm_r_fk_ctrl.Location.Y {type=MovieSceneScriptingFloatChannel, default=0, thinned_from=322, keys [[0,0],[354,0]]}
       upperarm_r_fk_ctrl.Location.Z {type=MovieSceneScriptingFloatChannel, default=0, thinned_from=322, keys [[0,0],[354,0]]}
       upperarm_r_fk_ctrl.Rotation.X {
        type=MovieSceneScriptingFloatChannel
        default=0
        thinned_from=322
        keys [[0,69.0489],[1,68.7225],[8,68.5892],[12,68.7187],[24,67.2586],[84,69.1717],[87,69.6331],[90,70.535],[103,76.8593],[108,78.4196],[116,79.4137],[125,79.9822],[138,82.0714],[201,82.4959],[215,80.6951],[220,79.1963],[235,70.9195],[239,69.2617],[244,67.9585],[250,67.3765],[298,65.0089],[320,67.2513],[354,69.0489]]
       }
       upperarm_r_fk_ctrl.Rotation.Y {
        type=MovieSceneScriptingFloatChannel
        default=0
        thinned_from=322
        keys [[0,-11.0059],[4,-11.8037],[8,-12.7385],[10,-13.2562],[12,-13.846],[19,-16.1632],[21,-16.7589],[23,-17.2744],[25,-17.6893],[27,-17.9864],[29,-18.1809],[31,-18.275],[52,-18.1872],[81,-19.082],[83,-19.1027],[84,-19.1548],[85,-19.2816],[86,-19.486],[87,-19.7905],[88,-20.1901],[89,-20.6631],[90,-21.2176],[91,-21.8516],[95,-24.6526],[96,-25.2809],[97,-25.8243],[98,-26.2562],[99,-26.5627],[100,-26.7565],[101,-26.8498],[102,-26.8417],[103,-26.737],[104,-26.5363],[105,-26.2375],[106,-25.843],[108,-24.8653],[113,-22.0772],[118,-19.4911],[119,-19.0073],[120,-18.5774],[121,-18.2143],[122,-17.9291],[123,-17.7109],[124,-17.5364],[125,-17.4162],[126,-17.3595],[127,-17.3558],[129,-17.4696],[134,-17.8864],[155,-19.0952],[163,-19.2799],[176,-19.1379],[179,-19.2736],[183,-19.5252],[185,-19.5647],[198,-18.9897],[204,-18.5169],[209,-17.9347],[213,-17.3337],[215,-16.9447],[216,-16.6893],[217,-16.385],[218,-16.0361],[219,-15.6339],[220,-15.1549],[221,-14.5855],[222,-13.9147],[223,-13.1634],[224,-12.3557],[225,-11.5147],[228,-8.9267],[229,-8.1073],[230,-7.3373],[231,-6.6411],[232,-6.0345],[233,-5.5344],[234,-5.1501],[235,-4.8876],[236,-4.7219],[237,-4.6264],[238,-4.592],[239,-4.6226],[240,-4.7095],[241,-4.8437],[242,-5.0251],[243,-5.2476],[244,-5.5099],[245,-5.8058],[246,-6.1359],[247,-6.4943],[248,-6.8783],[251,-8.1009],[256,-10.2142],[259,-11.4071],[261,-12.1448],[262,-12.477],[263,-12.7803],[264,-13.0398],[265,-13.2581],[266,-13.4273],[267,-13.5473],[268,-13.6239],[271,-13.6236],[284,-13.2997],[290,-13.2547],[293,-13.2994],[296,-13.4491],[298,-13.6145],[300,-13.8718],[302,-14.201],[306,-15.0103],[317,-17.4749],[320,-18.0355],[354,-11.0059]]
       }
       upperarm_r_fk_ctrl.Rotation.Z {
        type=MovieSceneScriptingFloatChannel
        default=0
        thinned_from=322
        keys [[0,52.7544],[10,54.1201],[13,54.876],[24,59.271],[29,60.5397],[42,61.513],[57,61.0892],[96,57.5803],[98,57.1624],[101,56.1057],[110,51.7385],[113,50.5684],[116,49.7044],[118,49.3724],[120,49.2788],[122,49.4291],[126,50.2411],[132,51.7875],[136,52.3241],[145,52.0303],[171,50.1824],[180,50.4135],[206,52.033],[215,51.7508],[221,51.005],[223,50.5429],[232,46.9402],[239,45.0596],[249,43.6088],[256,43.3707],[268,44.1376],[291,44.8234],[314,44.7075],[320,45.0118],[354,52.7544]]
       }
       upperarm_r_fk_ctrl.Scale.X {type=MovieSceneScriptingFloatChannel, default=1, const [1,322]}
       upperarm_r_fk_ctrl.Scale.Y {type=MovieSceneScriptingFloatChannel, default=1, const [1,322]}
       upperarm_r_fk_ctrl.Scale.Z {type=MovieSceneScriptingFloatChannel, default=1, const [1,322]}
       lowerarm_r_fk_ctrl.Location.X {type=MovieSceneScriptingFloatChannel, default=0, thinned_from=322, keys [[0,0],[354,0]]}
       lowerarm_r_fk_ctrl.Location.Y {type=MovieSceneScriptingFloatChannel, default=0, thinned_from=322, keys [[0,0],[354,0]]}
       lowerarm_r_fk_ctrl.Location.Z {type=MovieSceneScriptingFloatChannel, default=0, thinned_from=322, keys [[0,0],[354,0]]}
       lowerarm_r_fk_ctrl.Rotation.X {
        type=MovieSceneScriptingFloatChannel
        default=0
        thinned_from=322
        keys [[0,-0.7327],[1,-0.7471],[2,-0.7507],[3,-0.7579],[7,-0.7723],[10,-0.7723],[11,-0.7759],[12,-0.7759],[13,-0.7723],[15,-0.7723],[16,-0.7687],[17,-0.7702],[20,-0.7702],[21,-0.7676],[24,-0.7676],[25,-0.7702],[28,-0.7702],[29,-0.7728],[31,-0.7728],[32,-0.7753],[33,-0.7745],[35,-0.7779],[37,-0.7779],[38,-0.7813],[41,-0.7813],[43,-0.7848],[48,-0.7848],[51,-0.7862],[56,-0.779],[57,-0.779],[65,-0.7623],[66,-0.7623],[67,-0.7579],[68,-0.7579],[69,-0.7535],[70,-0.7535],[71,-0.7491],[72,-0.7491],[73,-0.7447],[74,-0.7447],[75,-0.7403],[76,-0.7403],[77,-0.7359],[79,-0.7183],[81,-0.6925],[83,-0.6523],[84,-0.6255],[86,-0.5652],[90,-0.4312],[91,-0.401],[93,-0.3474],[96,-0.277],[97,-0.2609],[98,-0.2405],[100,-0.2134],[102,-0.1998],[103,-0.1862],[104,-0.1794],[105,-0.1794],[106,-0.1726],[110,-0.1726],[112,-0.1862],[113,-0.1916],[114,-0.2024],[115,-0.2149],[116,-0.2293],[118,-0.2651],[120,-0.2992],[121,-0.3171],[122,-0.3332],[125,-0.3763],[127,-0.3996],[128,-0.4103],[129,-0.4175],[131,-0.4399],[138,-0.479],[139,-0.479],[140,-0.4846],[141,-0.4846],[142,-0.4902],[144,-0.4902],[145,-0.4942],[148,-0.4942],[149,-0.4974],[150,-0.4974],[151,-0.499],[159,-0.499],[160,-0.5031],[161,-0.5023],[162,-0.5066],[163,-0.5065],[165,-0.5108],[168,-0.5108],[169,-0.5065],[171,-0.5066],[172,-0.5023],[173,-0.5023],[175,-0.4937],[176,-0.4937],[178,-0.4882],[181,-0.4876],[185,-0.4866],[192,-0.4962],[193,-0.499],[194,-0.499],[196,-0.5062],[197,-0.5062],[200,-0.517],[201,-0.517],[208,-0.5422],[209,-0.5435],[210,-0.5483],[212,-0.5531],[214,-0.5602],[215,-0.5602],[216,-0.5578],[217,-0.5531],[218,-0.5459],[219,-0.534],[220,-0.5198],[221,-0.5007],[222,-0.4793],[223,-0.4532],[224,-0.4223],[225,-0.3923],[227,-0.32],[229,-0.2333],[230,-0.1971],[232,-0.1104],[233,-0.0742],[235,-0.0164],[237,0.027],[238,0.0342],[239,0.0487],[240,0.0559],[241,0.0613],[242,0.0699],[247,0.1056],[248,0.1113],[251,0.1241],[256,0.1384],[257,0.1381],[258,0.1424],[262,0.1424],[263,0.1467],[265,0.1467],[266,0.1509],[269,0.1509],[270,0.1552],[271,0.1552],[272,0.1595],[273,0.1578],[274,0.1618],[276,0.1618],[277,0.1658],[281,0.1658],[282,0.1698],[309,0.1698],[310,0.1638],[313,0.1638],[314,0.1578],[318,0.1578],[319,0.1518],[320,0.1518],[354,-0.7327]]
       }
       lowerarm_r_fk_ctrl.Rotation.Y {type=MovieSceneScriptingFloatChannel, default=0, thinned_from=322, keys [[0,-3.3401],[354,-3.3401]]}
       lowerarm_r_fk_ctrl.Rotation.Z {type=MovieSceneScriptingFloatChannel, default=0, thinned_from=322, keys [[0,76.2805],[354,76.2805]]}
       lowerarm_r_fk_ctrl.Scale.X {type=MovieSceneScriptingFloatChannel, default=1, const [1,322]}
       lowerarm_r_fk_ctrl.Scale.Y {type=MovieSceneScriptingFloatChannel, default=1, const [1,322]}
       lowerarm_r_fk_ctrl.Scale.Z {type=MovieSceneScriptingFloatChannel, default=1, const [1,322]}
       hand_r_fk_ctrl.Location.X {type=MovieSceneScriptingFloatChannel, default=0, thinned_from=322, keys [[0,0],[354,0]]}
       hand_r_fk_ctrl.Location.Y {type=MovieSceneScriptingFloatChannel, default=0, thinned_from=322, keys [[0,0],[354,0]]}
       hand_r_fk_ctrl.Location.Z {type=MovieSceneScriptingFloatChannel, default=0, thinned_from=322, keys [[0,0],[354,0]]}
       hand_r_fk_ctrl.Rotation.X {type=MovieSceneScriptingFloatChannel, default=0, thinned_from=322, keys [[0,-34.4342],[225,-34.7901],[320,-35.3527],[354,-34.4342]]}
       hand_r_fk_ctrl.Rotation.Y {type=MovieSceneScriptingFloatChannel, default=0, thinned_from=322, keys [[0,19.5362],[320,19.3453],[354,19.5362]]}
       hand_r_fk_ctrl.Rotation.Z {type=MovieSceneScriptingFloatChannel, default=0, thinned_from=322, keys [[0,-11.1755],[226,-11.3066],[320,-11.4834],[354,-11.1755]]}
       hand_r_fk_ctrl.Scale.X {type=MovieSceneScriptingFloatChannel, default=1, const [1,322]}
       hand_r_fk_ctrl.Scale.Y {type=MovieSceneScriptingFloatChannel, default=1, const [1,322]}
       hand_r_fk_ctrl.Scale.Z {type=MovieSceneScriptingFloatChannel, default=1, const [1,322]}
       thumb_01_r_ctrl.Location.X {type=MovieSceneScriptingFloatChannel, default=0, const [1.5173,322]}
       thumb_01_r_ctrl.Location.Y {type=MovieSceneScriptingFloatChannel, default=0, const [-0.0715,322]}
       thumb_01_r_ctrl.Location.Z {type=MovieSceneScriptingFloatChannel, default=0, const [0.0001,322]}
       thumb_01_r_ctrl.Rotation.X {type=MovieSceneScriptingFloatChannel, default=0, const [-33.464,322]}
       thumb_01_r_ctrl.Rotation.Y {type=MovieSceneScriptingFloatChannel, default=0, const [-13.0752,322]}
       thumb_01_r_ctrl.Rotation.Z {type=MovieSceneScriptingFloatChannel, default=0, const [-9.394,322]}
       thumb_01_r_ctrl.Scale.X {type=MovieSceneScriptingFloatChannel, default=1, const [1,322]}
       thumb_01_r_ctrl.Scale.Y {type=MovieSceneScriptingFloatChannel, default=1, const [1,322]}
       thumb_01_r_ctrl.Scale.Z {type=MovieSceneScriptingFloatChannel, default=1, const [1,322]}
       thumb_02_r_ctrl.Location.X {type=MovieSceneScriptingFloatChannel, default=0, thinned_from=322, keys [[0,0],[354,0]]}
       thumb_02_r_ctrl.Location.Y {type=MovieSceneScriptingFloatChannel, default=0, thinned_from=322, keys [[0,0],[354,0]]}
       thumb_02_r_ctrl.Location.Z {type=MovieSceneScriptingFloatChannel, default=0, thinned_from=322, keys [[0,0],[354,0]]}
       thumb_02_r_ctrl.Rotation.X {type=MovieSceneScriptingFloatChannel, default=0, const [1.2228,322]}
       thumb_02_r_ctrl.Rotation.Y {type=MovieSceneScriptingFloatChannel, default=0, const [-2.3382,322]}
       thumb_02_r_ctrl.Rotation.Z {type=MovieSceneScriptingFloatChannel, default=0, const [-13.4617,322]}
       thumb_02_r_ctrl.Scale.X {type=MovieSceneScriptingFloatChannel, default=1, const [1,322]}
       thumb_02_r_ctrl.Scale.Y {type=MovieSceneScriptingFloatChannel, default=1, const [1,322]}
       thumb_02_r_ctrl.Scale.Z {type=MovieSceneScriptingFloatChannel, default=1, const [1,322]}
       thumb_03_r_ctrl.Location.X {type=MovieSceneScriptingFloatChannel, default=0, thinned_from=322, keys [[0,0],[354,0]]}
       thumb_03_r_ctrl.Location.Y {type=MovieSceneScriptingFloatChannel, default=0, thinned_from=322, keys [[0,0],[354,0]]}
       thumb_03_r_ctrl.Location.Z {type=MovieSceneScriptingFloatChannel, default=0, thinned_from=322, keys [[0,0],[354,0]]}
       thumb_03_r_ctrl.Rotation.X {type=MovieSceneScriptingFloatChannel, default=0, const [2.0016,322]}
       thumb_03_r_ctrl.Rotation.Y {type=MovieSceneScriptingFloatChannel, default=0, const [0.1771,322]}
       thumb_03_r_ctrl.Rotation.Z {type=MovieSceneScriptingFloatChannel, default=0, const [-52.3193,322]}
       thumb_03_r_ctrl.Scale.X {type=MovieSceneScriptingFloatChannel, default=1, const [1,322]}
       thumb_03_r_ctrl.Scale.Y {type=MovieSceneScriptingFloatChannel, default=1, const [1,322]}
       thumb_03_r_ctrl.Scale.Z {type=MovieSceneScriptingFloatChannel, default=1, const [1,322]}
       index_metacarpal_r_ctrl.Location.X {type=MovieSceneScriptingFloatChannel, default=0, thinned_from=322, keys [[0,0],[354,0]]}
       index_metacarpal_r_ctrl.Location.Y {type=MovieSceneScriptingFloatChannel, default=0, thinned_from=322, keys [[0,0],[354,0]]}
       index_metacarpal_r_ctrl.Location.Z {type=MovieSceneScriptingFloatChannel, default=0, thinned_from=322, keys [[0,0],[354,0]]}
       index_metacarpal_r_ctrl.Rotation.X {type=MovieSceneScriptingFloatChannel, default=0, const [0,322]}
       index_metacarpal_r_ctrl.Rotation.Y {type=MovieSceneScriptingFloatChannel, default=0, const [4.1035,322]}
       index_metacarpal_r_ctrl.Rotation.Z {type=MovieSceneScriptingFloatChannel, default=0, thinned_from=322, keys [[0,0],[354,0]]}
       index_metacarpal_r_ctrl.Scale.X {type=MovieSceneScriptingFloatChannel, default=1, const [1,322]}
       index_metacarpal_r_ctrl.Scale.Y {type=MovieSceneScriptingFloatChannel, default=1, const [1,322]}
       index_metacarpal_r_ctrl.Scale.Z {type=MovieSceneScriptingFloatChannel, default=1, const [1,322]}
       index_01_r_ctrl.Location.X {type=MovieSceneScriptingFloatChannel, default=0, thinned_from=322, keys [[0,0],[354,0]]}
       index_01_r_ctrl.Location.Y {type=MovieSceneScriptingFloatChannel, default=0, thinned_from=322, keys [[0,0],[354,0]]}
       index_01_r_ctrl.Location.Z {type=MovieSceneScriptingFloatChannel, default=0, thinned_from=322, keys [[0,0],[354,0]]}
       index_01_r_ctrl.Rotation.X {type=MovieSceneScriptingFloatChannel, default=0, const [7.7952,322]}
       index_01_r_ctrl.Rotation.Y {type=MovieSceneScriptingFloatChannel, default=0, const [-9.2432,322]}
       index_01_r_ctrl.Rotation.Z {type=MovieSceneScriptingFloatChannel, default=0, const [-26.0805,322]}
       index_01_r_ctrl.Scale.X {type=MovieSceneScriptingFloatChannel, default=1, const [1,322]}
       index_01_r_ctrl.Scale.Y {type=MovieSceneScriptingFloatChannel, default=1, const [1,322]}
       index_01_r_ctrl.Scale.Z {type=MovieSceneScriptingFloatChannel, default=1, const [1,322]}
       index_02_r_ctrl.Location.X {type=MovieSceneScriptingFloatChannel, default=0, const [0.3298,322]}
       index_02_r_ctrl.Location.Y {type=MovieSceneScriptingFloatChannel, default=0, const [-0.1751,322]}
       index_02_r_ctrl.Location.Z {type=MovieSceneScriptingFloatChannel, default=0, const [-0.0001,322]}
       index_02_r_ctrl.Rotation.X {type=MovieSceneScriptingFloatChannel, default=0, const [-0.379,322]}
       index_02_r_ctrl.Rotation.Y {type=MovieSceneScriptingFloatChannel, default=0, const [-0.0975,322]}
       index_02_r_ctrl.Rotation.Z {type=MovieSceneScriptingFloatChannel, default=0, const [0,322]}
       index_02_r_ctrl.Scale.X {type=MovieSceneScriptingFloatChannel, default=1, const [1,322]}
       index_02_r_ctrl.Scale.Y {type=MovieSceneScriptingFloatChannel, default=1, const [1,322]}
       index_02_r_ctrl.Scale.Z {type=MovieSceneScriptingFloatChannel, default=1, const [1,322]}
       index_03_r_ctrl.Location.X {type=MovieSceneScriptingFloatChannel, default=0, const [0.2932,322]}
       index_03_r_ctrl.Location.Y {type=MovieSceneScriptingFloatChannel, default=0, const [0.1198,322]}
       index_03_r_ctrl.Location.Z {type=MovieSceneScriptingFloatChannel, default=0, const [0.005,322]}
       index_03_r_ctrl.Rotation.X {type=MovieSceneScriptingFloatChannel, default=0, const [0.0155,322]}
       index_03_r_ctrl.Rotation.Y {type=MovieSceneScriptingFloatChannel, default=0, const [0.793,322]}
       index_03_r_ctrl.Rotation.Z {type=MovieSceneScriptingFloatChannel, default=0, const [33.9964,322]}
       index_03_r_ctrl.Scale.X {type=MovieSceneScriptingFloatChannel, default=1, const [1,322]}
       index_03_r_ctrl.Scale.Y {type=MovieSceneScriptingFloatChannel, default=1, const [1,322]}
       index_03_r_ctrl.Scale.Z {type=MovieSceneScriptingFloatChannel, default=1, const [1,322]}
       middle_metacarpal_r_ctrl.Location.X {type=MovieSceneScriptingFloatChannel, default=0, thinned_from=322, keys [[0,0],[354,0]]}
       middle_metacarpal_r_ctrl.Location.Y {type=MovieSceneScriptingFloatChannel, default=0, thinned_from=322, keys [[0,0],[354,0]]}
       middle_metacarpal_r_ctrl.Location.Z {type=MovieSceneScriptingFloatChannel, default=0, thinned_from=322, keys [[0,0],[354,0]]}
       middle_metacarpal_r_ctrl.Rotation.X {type=MovieSceneScriptingFloatChannel, default=0, const [0,322]}
       middle_metacarpal_r_ctrl.Rotation.Y {type=MovieSceneScriptingFloatChannel, default=0, const [1.2252,322]}
       middle_metacarpal_r_ctrl.Rotation.Z {type=MovieSceneScriptingFloatChannel, default=0, thinned_from=322, keys [[0,0],[354,0]]}
       middle_metacarpal_r_ctrl.Scale.X {type=MovieSceneScriptingFloatChannel, default=1, const [1,322]}
       middle_metacarpal_r_ctrl.Scale.Y {type=MovieSceneScriptingFloatChannel, default=1, const [1,322]}
       middle_metacarpal_r_ctrl.Scale.Z {type=MovieSceneScriptingFloatChannel, default=1, const [1,322]}
       middle_01_r_ctrl.Location.X {type=MovieSceneScriptingFloatChannel, default=0, thinned_from=322, keys [[0,0],[354,0]]}
       middle_01_r_ctrl.Location.Y {type=MovieSceneScriptingFloatChannel, default=0, thinned_from=322, keys [[0,0],[354,0]]}
       middle_01_r_ctrl.Location.Z {type=MovieSceneScriptingFloatChannel, default=0, thinned_from=322, keys [[0,0],[354,0]]}
       middle_01_r_ctrl.Rotation.X {type=MovieSceneScriptingFloatChannel, default=0, const [3.7685,322]}
       middle_01_r_ctrl.Rotation.Y {type=MovieSceneScriptingFloatChannel, default=0, const [-4.0758,322]}
       middle_01_r_ctrl.Rotation.Z {type=MovieSceneScriptingFloatChannel, default=0, const [-28.6993,322]}
       middle_01_r_ctrl.Scale.X {type=MovieSceneScriptingFloatChannel, default=1, const [1,322]}
       middle_01_r_ctrl.Scale.Y {type=MovieSceneScriptingFloatChannel, default=1, const [1,322]}
       middle_01_r_ctrl.Scale.Z {type=MovieSceneScriptingFloatChannel, default=1, const [1,322]}
       middle_02_r_ctrl.Location.X {type=MovieSceneScriptingFloatChannel, default=0, const [0.0008,322]}
       middle_02_r_ctrl.Location.Y {type=MovieSceneScriptingFloatChannel, default=0, const [0.0009,322]}
       middle_02_r_ctrl.Location.Z {type=MovieSceneScriptingFloatChannel, default=0, const [0.0476,322]}
       middle_02_r_ctrl.Rotation.X {type=MovieSceneScriptingFloatChannel, default=0, const [-3.8525,322]}
       middle_02_r_ctrl.Rotation.Y {type=MovieSceneScriptingFloatChannel, default=0, const [-0.3785,322]}
       middle_02_r_ctrl.Rotation.Z {type=MovieSceneScriptingFloatChannel, default=0, const [0,322]}
       middle_02_r_ctrl.Scale.X {type=MovieSceneScriptingFloatChannel, default=1, const [1,322]}
       middle_02_r_ctrl.Scale.Y {type=MovieSceneScriptingFloatChannel, default=1, const [1,322]}
       middle_02_r_ctrl.Scale.Z {type=MovieSceneScriptingFloatChannel, default=1, const [1,322]}
       middle_03_r_ctrl.Location.X {type=MovieSceneScriptingFloatChannel, default=0, thinned_from=322, keys [[0,0],[354,0]]}
       middle_03_r_ctrl.Location.Y {type=MovieSceneScriptingFloatChannel, default=0, thinned_from=322, keys [[0,0],[354,0]]}
       middle_03_r_ctrl.Location.Z {type=MovieSceneScriptingFloatChannel, default=0, thinned_from=322, keys [[0,0],[354,0]]}
       middle_03_r_ctrl.Rotation.X {type=MovieSceneScriptingFloatChannel, default=0, const [1.5318,322]}
       middle_03_r_ctrl.Rotation.Y {type=MovieSceneScriptingFloatChannel, default=0, const [1.7844,322]}
       middle_03_r_ctrl.Rotation.Z {type=MovieSceneScriptingFloatChannel, default=0, const [22.205,322]}
       middle_03_r_ctrl.Scale.X {type=MovieSceneScriptingFloatChannel, default=1, const [1,322]}
       middle_03_r_ctrl.Scale.Y {type=MovieSceneScriptingFloatChannel, default=1, const [1,322]}
       middle_03_r_ctrl.Scale.Z {type=MovieSceneScriptingFloatChannel, default=1, const [1,322]}
       ring_metacarpal_r_ctrl.Location.X {type=MovieSceneScriptingFloatChannel, default=0, thinned_from=322, keys [[0,0],[354,0]]}
       ring_metacarpal_r_ctrl.Location.Y {type=MovieSceneScriptingFloatChannel, default=0, thinned_from=322, keys [[0,0],[354,0]]}
       ring_metacarpal_r_ctrl.Location.Z {type=MovieSceneScriptingFloatChannel, default=0, thinned_from=322, keys [[0,0],[354,0]]}
       ring_metacarpal_r_ctrl.Rotation.X {type=MovieSceneScriptingFloatChannel, default=0, const [0,322]}
       ring_metacarpal_r_ctrl.Rotation.Y {type=MovieSceneScriptingFloatChannel, default=0, const [-4.4979,322]}
       ring_metacarpal_r_ctrl.Rotation.Z {type=MovieSceneScriptingFloatChannel, default=0, const [0,322]}
       ring_metacarpal_r_ctrl.Scale.X {type=MovieSceneScriptingFloatChannel, default=1, const [1,322]}
       ring_metacarpal_r_ctrl.Scale.Y {type=MovieSceneScriptingFloatChannel, default=1, const [1,322]}
       ring_metacarpal_r_ctrl.Scale.Z {type=MovieSceneScriptingFloatChannel, default=1, const [1,322]}
       ring_01_r_ctrl.Location.X {type=MovieSceneScriptingFloatChannel, default=0, thinned_from=322, keys [[0,0],[354,0]]}
       ring_01_r_ctrl.Location.Y {type=MovieSceneScriptingFloatChannel, default=0, thinned_from=322, keys [[0,0],[354,0]]}
       ring_01_r_ctrl.Location.Z {type=MovieSceneScriptingFloatChannel, default=0, thinned_from=322, keys [[0,0],[354,0]]}
       ring_01_r_ctrl.Rotation.X {type=MovieSceneScriptingFloatChannel, default=0, const [-3.1309,322]}
       ring_01_r_ctrl.Rotation.Y {type=MovieSceneScriptingFloatChannel, default=0, const [4.6931,322]}
       ring_01_r_ctrl.Rotation.Z {type=MovieSceneScriptingFloatChannel, default=0, const [-25.3784,322]}
       ring_01_r_ctrl.Scale.X {type=MovieSceneScriptingFloatChannel, default=1, const [1,322]}
       ring_01_r_ctrl.Scale.Y {type=MovieSceneScriptingFloatChannel, default=1, const [1,322]}
       ring_01_r_ctrl.Scale.Z {type=MovieSceneScriptingFloatChannel, default=1, const [1,322]}
       ring_02_r_ctrl.Location.X {type=MovieSceneScriptingFloatChannel, default=0, const [0,322]}
       ring_02_r_ctrl.Location.Y {type=MovieSceneScriptingFloatChannel, default=0, const [0,322]}
       ring_02_r_ctrl.Location.Z {type=MovieSceneScriptingFloatChannel, default=0, const [0,322]}
       ring_02_r_ctrl.Rotation.X {type=MovieSceneScriptingFloatChannel, default=0, const [0,322]}
       ring_02_r_ctrl.Rotation.Y {type=MovieSceneScriptingFloatChannel, default=0, const [0,322]}
       ring_02_r_ctrl.Rotation.Z {type=MovieSceneScriptingFloatChannel, default=0, const [0,322]}
       ring_02_r_ctrl.Scale.X {type=MovieSceneScriptingFloatChannel, default=1, const [1,322]}
       ring_02_r_ctrl.Scale.Y {type=MovieSceneScriptingFloatChannel, default=1, const [1,322]}
       ring_02_r_ctrl.Scale.Z {type=MovieSceneScriptingFloatChannel, default=1, const [1,322]}
       ring_03_r_ctrl.Location.X {type=MovieSceneScriptingFloatChannel, default=0, thinned_from=322, keys [[0,0],[354,0]]}
       ring_03_r_ctrl.Location.Y {type=MovieSceneScriptingFloatChannel, default=0, thinned_from=322, keys [[0,0],[354,0]]}
       ring_03_r_ctrl.Location.Z {type=MovieSceneScriptingFloatChannel, default=0, thinned_from=322, keys [[0,0],[354,0]]}
       ring_03_r_ctrl.Rotation.X {type=MovieSceneScriptingFloatChannel, default=0, thinned_from=322, keys [[0,0],[354,0]]}
       ring_03_r_ctrl.Rotation.Y {type=MovieSceneScriptingFloatChannel, default=0, const [0,322]}
       ring_03_r_ctrl.Rotation.Z {type=MovieSceneScriptingFloatChannel, default=0, const [22.1957,322]}
       ring_03_r_ctrl.Scale.X {type=MovieSceneScriptingFloatChannel, default=1, const [1,322]}
       ring_03_r_ctrl.Scale.Y {type=MovieSceneScriptingFloatChannel, default=1, const [1,322]}
       ring_03_r_ctrl.Scale.Z {type=MovieSceneScriptingFloatChannel, default=1, const [1,322]}
       pinky_metacarpal_r_ctrl.Location.X {type=MovieSceneScriptingFloatChannel, default=0, thinned_from=322, keys [[0,0],[354,0]]}
       pinky_metacarpal_r_ctrl.Location.Y {type=MovieSceneScriptingFloatChannel, default=0, thinned_from=322, keys [[0,0],[354,0]]}
       pinky_metacarpal_r_ctrl.Location.Z {type=MovieSceneScriptingFloatChannel, default=0, thinned_from=322, keys [[0,0],[354,0]]}
       pinky_metacarpal_r_ctrl.Rotation.X {type=MovieSceneScriptingFloatChannel, default=0, const [-4.938,322]}
       pinky_metacarpal_r_ctrl.Rotation.Y {type=MovieSceneScriptingFloatChannel, default=0, const [-4.2094,322]}
       pinky_metacarpal_r_ctrl.Rotation.Z {type=MovieSceneScriptingFloatChannel, default=0, const [-2.0167,322]}
       pinky_metacarpal_r_ctrl.Scale.X {type=MovieSceneScriptingFloatChannel, default=1, const [1,322]}
       pinky_metacarpal_r_ctrl.Scale.Y {type=MovieSceneScriptingFloatChannel, default=1, const [1,322]}
       pinky_metacarpal_r_ctrl.Scale.Z {type=MovieSceneScriptingFloatChannel, default=1, const [1,322]}
       pinky_01_r_ctrl.Location.X {type=MovieSceneScriptingFloatChannel, default=0, thinned_from=322, keys [[0,0],[354,0]]}
       pinky_01_r_ctrl.Location.Y {type=MovieSceneScriptingFloatChannel, default=0, thinned_from=322, keys [[0,0],[354,0]]}
       pinky_01_r_ctrl.Location.Z {type=MovieSceneScriptingFloatChannel, default=0, thinned_from=322, keys [[0,0],[354,0]]}
       pinky_01_r_ctrl.Rotation.X {type=MovieSceneScriptingFloatChannel, default=0, const [-1.5748,322]}
       pinky_01_r_ctrl.Rotation.Y {type=MovieSceneScriptingFloatChannel, default=0, const [7.2218,322]}
       pinky_01_r_ctrl.Rotation.Z {type=MovieSceneScriptingFloatChannel, default=0, const [-7.7286,322]}
       pinky_01_r_ctrl.Scale.X {type=MovieSceneScriptingFloatChannel, default=1, const [1,322]}
       pinky_01_r_ctrl.Scale.Y {type=MovieSceneScriptingFloatChannel, default=1, const [1,322]}
       pinky_01_r_ctrl.Scale.Z {type=MovieSceneScriptingFloatChannel, default=1, const [1,322]}
       pinky_02_r_ctrl.Location.X {type=MovieSceneScriptingFloatChannel, default=0, thinned_from=322, keys [[0,0],[354,0]]}
       pinky_02_r_ctrl.Location.Y {type=MovieSceneScriptingFloatChannel, default=0, thinned_from=322, keys [[0,0],[354,0]]}
       pinky_02_r_ctrl.Location.Z {type=MovieSceneScriptingFloatChannel, default=0, thinned_from=322, keys [[0,0],[354,0]]}
       pinky_02_r_ctrl.Rotation.X {type=MovieSceneScriptingFloatChannel, default=0, const [-0.0384,322]}
       pinky_02_r_ctrl.Rotation.Y {type=MovieSceneScriptingFloatChannel, default=0, const [3.6728,322]}
       pinky_02_r_ctrl.Rotation.Z {type=MovieSceneScriptingFloatChannel, default=0, const [0,322]}
       pinky_02_r_ctrl.Scale.X {type=MovieSceneScriptingFloatChannel, default=1, const [1,322]}
       pinky_02_r_ctrl.Scale.Y {type=MovieSceneScriptingFloatChannel, default=1, const [1,322]}
       pinky_02_r_ctrl.Scale.Z {type=MovieSceneScriptingFloatChannel, default=1, const [1,322]}
       pinky_03_r_ctrl.Location.X {type=MovieSceneScriptingFloatChannel, default=0, thinned_from=322, keys [[0,0],[354,0]]}
       pinky_03_r_ctrl.Location.Y {type=MovieSceneScriptingFloatChannel, default=0, thinned_from=322, keys [[0,0],[354,0]]}
       pinky_03_r_ctrl.Location.Z {type=MovieSceneScriptingFloatChannel, default=0, thinned_from=322, keys [[0,0],[354,0]]}
       pinky_03_r_ctrl.Rotation.X {type=MovieSceneScriptingFloatChannel, default=0, const [-0.1956,322]}
       pinky_03_r_ctrl.Rotation.Y {type=MovieSceneScriptingFloatChannel, default=0, const [1.1092,322]}
       pinky_03_r_ctrl.Rotation.Z {type=MovieSceneScriptingFloatChannel, default=0, const [0,322]}
       pinky_03_r_ctrl.Scale.X {type=MovieSceneScriptingFloatChannel, default=1, const [1,322]}
       pinky_03_r_ctrl.Scale.Y {type=MovieSceneScriptingFloatChannel, default=1, const [1,322]}
       pinky_03_r_ctrl.Scale.Z {type=MovieSceneScriptingFloatChannel, default=1, const [1,322]}
       hand_r_ik_ctrl.Location.X {
        type=MovieSceneScriptingFloatChannel
        default=0
        thinned_from=322
        keys [[0,-66.0294],[1,-66.5896],[5,-66.9846],[12,-66.4258],[30,-63.7278],[39,-63.8588],[68,-66.2832],[95,-63.5821],[102,-63.804],[134,-70.0462],[143,-70.545],[203,-66.9846],[218,-67.1786],[223,-66.5486],[244,-60.3255],[251,-59.1609],[306,-60.2192],[320,-59.3299],[354,-66.0294]]
       }
       hand_r_ik_ctrl.Location.Y {
        type=MovieSceneScriptingFloatChannel
        default=0
        thinned_from=322
        keys [[0,21.3873],[7,21.7868],[24,23.5476],[29,23.7913],[52,23.1845],[66,23.4751],[77,24.2716],[87,25.1929],[97,25.4655],[102,25.3429],[106,25.0234],[112,24.2351],[128,21.4627],[133,20.7733],[138,20.3315],[142,20.1644],[154,20.3054],[177,21.2027],[205,22.6349],[216,22.7649],[220,22.6235],[223,22.359],[225,22.0885],[227,21.728],[230,21.0207],[234,19.8322],[240,17.6809],[245,15.848],[247,15.1986],[249,14.6393],[251,14.174],[253,13.8205],[256,13.4506],[259,13.1637],[266,12.6695],[272,12.3997],[281,12.1541],[288,12.1024],[292,12.1509],[293,12.179],[295,12.2918],[297,12.4656],[299,12.6889],[302,13.1266],[306,13.83],[318,16.0954],[320,16.4325],[354,21.3873]]
       }
       hand_r_ik_ctrl.Location.Z {
        type=MovieSceneScriptingFloatChannel
        default=0
        thinned_from=322
        keys [[0,40.3807],[21,37.6043],[30,36.9399],[40,37.171],[68,39.1594],[81,38.8121],[88,38.1587],[105,34.7527],[109,34.2859],[113,34.1382],[117,34.2636],[131,35.6747],[135,35.796],[140,35.5383],[168,31.5337],[183,30.2308],[196,30.1624],[205,30.6264],[214,31.7033],[218,32.4293],[237,38.0477],[241,38.894],[247,39.5504],[294,41.9326],[320,41.7076],[354,40.3807]]
       }
       hand_r_ik_ctrl.Rotation.X {
        type=MovieSceneScriptingFloatChannel
        default=0
        thinned_from=322
        keys [[0,77.891],[67,78.6556],[86,77.3027],[89,76.6079],[91,75.8089],[94,74.0836],[106,64.6241],[109,62.679],[112,61.1882],[115,60.1707],[118,59.5678],[121,59.3766],[127,59.9009],[135,61.133],[170,59.4496],[185,56.9758],[207,56.1507],[216,56.5269],[219,56.993],[226,59.1062],[230,60.7962],[236,64.3113],[242,68.5269],[249,73.88],[255,77.468],[265,82.1304],[283,86.1225],[295,87.0775],[320,84.4392],[354,77.891]]
       }
       hand_r_ik_ctrl.Rotation.Y {
        type=MovieSceneScriptingFloatChannel
        default=0
        thinned_from=322
        keys [[0,-6.5201],[1,-7.2762],[4,-8.1297],[9,-9.4459],[13,-10.4187],[23,-12.7937],[25,-13.2144],[27,-13.5654],[29,-13.8244],[31,-13.9856],[46,-14.3568],[54,-14.3445],[77,-13.5949],[80,-13.6028],[84,-13.7637],[87,-13.7773],[89,-13.8458],[90,-13.9366],[91,-14.0875],[92,-14.2948],[93,-14.5531],[94,-14.8467],[102,-17.6526],[104,-18.2628],[105,-18.5139],[108,-19.0757],[110,-19.3837],[113,-19.6941],[119,-20.0711],[123,-20.51],[130,-21.5848],[135,-22.167],[139,-22.4323],[162,-22.6158],[173,-23.1364],[176,-23.4231],[184,-24.698],[191,-25.362],[198,-25.6216],[203,-25.5591],[207,-25.2899],[210,-24.9128],[213,-24.3368],[215,-23.79],[216,-23.4481],[218,-22.6251],[219,-22.1364],[220,-21.5751],[221,-20.9496],[222,-20.2732],[223,-19.5248],[224,-18.701],[225,-17.8116],[226,-16.8591],[227,-15.8397],[228,-14.7584],[229,-13.6099],[230,-12.3962],[231,-11.1281],[232,-9.8026],[233,-8.423],[234,-6.9932],[235,-5.5258],[236,-4.0178],[237,-2.4813],[238,-0.9137],[239,0.6664],[240,2.2535],[241,3.8241],[242,5.3725],[243,6.8851],[244,8.3566],[245,9.7886],[246,11.1829],[247,12.5292],[248,13.8209],[249,15.0509],[250,16.2042],[251,17.2771],[252,18.2663],[253,19.1638],[254,19.9709],[255,20.7218],[256,21.4308],[258,22.7198],[260,23.8517],[262,24.8327],[264,25.6781],[266,26.3979],[268,26.9835],[272,27.8324],[276,28.4154],[283,28.9507],[292,29.2355],[294,29.205],[297,28.9499],[299,28.6243],[302,27.8909],[318,22.6063],[320,22.0176],[354,-6.5201]]
       }
       hand_r_ik_ctrl.Rotation.Z {
        type=MovieSceneScriptingFloatChannel
        default=0
        thinned_from=322
        keys [[0,142.0342],[29,146.78],[89,148.1946],[109,155.9058],[214,157.512],[219,156.0869],[241,140.0409],[249,136.6133],[298,134.648],[320,137.8525],[354,142.0342]]
       }
       hand_r_ik_ctrl.Scale.X {type=MovieSceneScriptingFloatChannel, default=1, const [1,322]}
       hand_r_ik_ctrl.Scale.Y {type=MovieSceneScriptingFloatChannel, default=1, const [1,322]}
       hand_r_ik_ctrl.Scale.Z {type=MovieSceneScriptingFloatChannel, default=1, const [1,322]}
       Softness_2 {type=MovieSceneScriptingFloatChannel, default=0, const [0,322]}
       hand_r_ikorient_ctrl.Rotation.Roll {type=MovieSceneScriptingFloatChannel, default=0, const [0,322]}
       hand_r_ikorient_ctrl.Rotation.Pitch {type=MovieSceneScriptingFloatChannel, default=0, const [0,322]}
       hand_r_ikorient_ctrl.Rotation.Yaw {type=MovieSceneScriptingFloatChannel, default=0, const [0,322]}
       arm_r_pv_ik_ctrl.Location.X {
        type=MovieSceneScriptingFloatChannel
        default=0
        thinned_from=322
        keys [[0,-4.962],[1,-5.4314],[6,-6.3082],[16,-7.9457],[20,-8.6089],[22,-8.9054],[24,-9.1546],[26,-9.3257],[28,-9.4267],[29,-9.4534],[30,-9.4609],[40,-9.2657],[45,-9.2797],[49,-9.3698],[52,-9.4743],[58,-9.7822],[80,-11.1789],[84,-11.5006],[89,-11.811],[90,-11.9129],[91,-12.0662],[92,-12.2583],[93,-12.5007],[94,-12.7681],[96,-13.4051],[103,-15.889],[104,-16.2158],[105,-16.4993],[106,-16.7409],[109,-17.2982],[111,-17.5369],[112,-17.6285],[115,-17.7307],[118,-17.6688],[127,-17.1955],[135,-16.5433],[140,-16.2854],[154,-16.1112],[170,-16.4287],[173,-16.5822],[176,-16.8611],[178,-17.1556],[184,-18.2557],[186,-18.5204],[195,-19.3219],[201,-19.6569],[205,-19.7403],[208,-19.6939],[211,-19.5346],[213,-19.3588],[215,-19.0835],[216,-18.8643],[217,-18.5847],[218,-18.2573],[219,-17.8573],[220,-17.3654],[221,-16.792],[222,-16.1656],[223,-15.4695],[224,-14.7021],[225,-13.8794],[226,-12.9901],[227,-12.0468],[228,-11.0525],[229,-10.0175],[230,-8.9534],[234,-4.5809],[235,-3.4966],[236,-2.4238],[237,-1.3632],[238,-0.3236],[239,0.6877],[240,1.6553],[241,2.5858],[242,3.4581],[243,4.2792],[244,5.0352],[245,5.7274],[246,6.3504],[247,6.9086],[248,7.3948],[249,7.8178],[250,8.1783],[251,8.4844],[252,8.7256],[253,8.9191],[254,9.0652],[256,9.2788],[258,9.4083],[259,9.4506],[262,9.48],[272,9.3019],[288,8.9815],[293,8.8243],[295,8.7145],[297,8.5706],[299,8.3978],[301,8.1864],[304,7.8176],[309,7.131],[320,5.5371],[354,-4.962]]
       }
       arm_r_pv_ik_ctrl.Location.Y {
        type=MovieSceneScriptingFloatChannel
        default=0
        thinned_from=322
        keys [[0,59.3404],[30,62.6224],[90,64.0574],[105,66.372],[135,67.0919],[172,64.5129],[216,66.2658],[222,65.829],[226,64.9917],[229,63.8888],[232,62.3105],[235,60.2655],[238,57.7058],[248,47.25],[250,45.3413],[252,43.6461],[254,42.2273],[257,40.4329],[260,38.9331],[264,37.3812],[267,36.4947],[272,35.5016],[278,34.8038],[292,34.3608],[295,34.4797],[298,34.8246],[301,35.4524],[320,41.3571],[354,59.3404]]
       }
       arm_r_pv_ik_ctrl.Location.Z {
        type=MovieSceneScriptingFloatChannel
        default=0
        thinned_from=322
        keys [[0,47.6037],[1,48.0711],[4,48.1924],[7,47.9368],[12,46.8758],[20,44.7311],[27,43.5356],[31,43.254],[37,43.5774],[57,46.0163],[66,46.3382],[71,46.0807],[84,43.8904],[96,40.9894],[100,40.3279],[105,39.9916],[108,40.1068],[112,40.6607],[116,41.548],[119,42.4851],[130,46.9376],[134,48.2035],[137,48.8059],[140,49.0339],[143,48.9356],[174,44.5224],[202,43.3947],[225,43.5768],[233,43.094],[250,40.6017],[255,40.3066],[275,41.3613],[288,42.0401],[296,41.8339],[310,40.2497],[320,38.5091],[354,47.6037]]
       }
       arm_r_pv_ik_ctrl.Rotation.X {
        type=MovieSceneScriptingFloatChannel
        default=0
        thinned_from=322
        keys [[0,-1.766],[1,-1.8281],[4,-1.998],[8,-2.2338],[9,-2.2893],[10,-2.3502],[11,-2.4336],[12,-2.5298],[13,-2.6424],[14,-2.7649],[15,-2.8994],[16,-3.044],[18,-3.3012],[19,-3.4114],[20,-3.5027],[21,-3.5814],[22,-3.6448],[23,-3.6907],[24,-3.7161],[25,-3.7243],[26,-3.7173],[37,-3.5061],[41,-3.4076],[45,-3.3277],[48,-3.2958],[50,-3.2892],[52,-3.2948],[53,-3.3208],[55,-3.3548],[58,-3.4338],[60,-3.5022],[61,-3.5432],[62,-3.576],[70,-3.9109],[74,-4.0527],[79,-4.1673],[81,-4.1955],[83,-4.2112],[84,-4.2284],[87,-4.3121],[88,-4.3516],[90,-4.447],[93,-4.553],[95,-4.5969],[103,-4.6869],[105,-4.6924],[106,-4.6864],[108,-4.6495],[113,-4.4871],[115,-4.4382],[116,-4.4004],[117,-4.3446],[118,-4.2733],[120,-4.0935],[121,-3.9968],[123,-3.7817],[126,-3.4181],[128,-3.1513],[133,-2.4431],[134,-2.3086],[135,-2.1878],[136,-2.0832],[137,-1.9958],[138,-1.9319],[139,-1.89],[140,-1.8663],[141,-1.8561],[142,-1.8534],[143,-1.8591],[144,-1.8712],[146,-1.9203],[149,-2.0208],[150,-2.0611],[152,-2.1289],[155,-2.2119],[157,-2.2435],[159,-2.2564],[161,-2.2476],[162,-2.2369],[163,-2.2208],[166,-2.1383],[169,-2.0405],[172,-1.9251],[174,-1.8343],[176,-1.7329],[177,-1.6789],[178,-1.6306],[179,-1.5785],[181,-1.4849],[182,-1.4451],[185,-1.3126],[187,-1.2184],[188,-1.175],[189,-1.1339],[190,-1.0986],[191,-1.0655],[192,-1.0367],[193,-1.0128],[194,-0.9933],[195,-0.9803],[197,-0.9622],[198,-0.9566],[200,-0.9572],[202,-0.9672],[203,-0.9767],[204,-0.9883],[206,-1.0227],[208,-1.0717],[209,-1.1014],[210,-1.1361],[211,-1.1812],[212,-1.2289],[213,-1.2846],[214,-1.3463],[215,-1.415],[216,-1.4733],[217,-1.5133],[218,-1.5373],[219,-1.539],[220,-1.5169],[221,-1.4704],[222,-1.3997],[223,-1.3241],[224,-1.2392],[225,-1.1459],[228,-0.8398],[229,-0.7482],[230,-0.6703],[231,-0.6167],[232,-0.5932],[233,-0.6112],[234,-0.6665],[235,-0.7567],[236,-0.8602],[237,-0.9729],[238,-1.0895],[239,-1.2139],[241,-1.478],[242,-1.6176],[243,-1.7611],[244,-1.9097],[245,-2.0674],[246,-2.2337],[249,-2.7515],[250,-2.917],[252,-3.2209],[253,-3.3602],[254,-3.4867],[255,-3.599],[257,-3.775],[259,-3.9722],[260,-4.0839],[264,-4.5716],[266,-4.7688],[268,-4.9242],[271,-5.0892],[275,-5.2559],[280,-5.3963],[298,-6.1519],[305,-6.5151],[310,-6.721],[314,-6.9427],[316,-7.087],[320,-7.4306],[354,-1.766]]
       }
       arm_r_pv_ik_ctrl.Rotation.Y {
        type=MovieSceneScriptingFloatChannel
        default=0
        thinned_from=322
        keys [[0,28.65],[1,29.1585],[5,29.879],[13,30.8317],[24,32.796],[33,33.6887],[51,34.1731],[85,32.4271],[87,32.2059],[90,31.6179],[94,30.4895],[99,28.7215],[107,25.5222],[109,24.8211],[112,23.9606],[115,23.3272],[117,23.0476],[119,22.9047],[121,22.9081],[123,23.0594],[126,23.4826],[131,24.3787],[134,24.7769],[138,25.0678],[145,25.0921],[173,23.9717],[183,23.5107],[205,23.1943],[210,22.8941],[214,22.4247],[216,22.0852],[219,21.4204],[221,20.8693],[222,20.5545],[224,19.8032],[229,17.7216],[231,16.97],[232,16.6379],[233,16.3439],[235,15.8427],[237,15.4165],[239,15.0493],[240,14.9023],[241,14.8],[242,14.7435],[243,14.7391],[250,15.1531],[254,15.5511],[258,16.0734],[265,17.2588],[269,17.7933],[282,19.256],[289,19.7385],[295,19.8669],[313,19.7232],[320,19.8899],[354,28.65]]
       }
       arm_r_pv_ik_ctrl.Rotation.Z {
        type=MovieSceneScriptingFloatChannel
        default=0
        thinned_from=322
        keys [[0,102.9282],[1,103.571],[28,109.1287],[87,109.9197],[90,110.5449],[93,111.9734],[106,121.6548],[112,124.4804],[176,129.3506],[200,132.7654],[213,131.4472],[218,129.5753],[222,126.75],[226,122.7352],[230,117.4821],[235,109.3603],[245,91.0485],[248,86.0842],[250,83.1443],[252,80.5637],[255,77.4097],[259,73.9682],[263,71.154],[267,69.0169],[274,66.7326],[284,65.0146],[293,64.3759],[297,64.6674],[301,65.584],[320,72.6682],[354,102.9282]]
       }
       arm_r_pv_ik_ctrl.Scale.X {type=MovieSceneScriptingFloatChannel, default=1, const [1,322]}
       arm_r_pv_ik_ctrl.Scale.Y {type=MovieSceneScriptingFloatChannel, default=1, const [1,322]}
       arm_r_pv_ik_ctrl.Scale.Z {type=MovieSceneScriptingFloatChannel, default=1, const [1,322]}
       body_ctrl.Location.X {
        type=MovieSceneScriptingFloatChannel
        default=0
        thinned_from=315
        keys [[0,6.6514],[8,6.6042],[11,6.5357],[14,6.5267],[19,6.5972],[23,6.6723],[26,6.6723],[39,6.5163],[43,6.5214],[48,6.6053],[52,6.7296],[64,7.265],[74,7.7251],[78,7.8656],[82,7.9641],[86,7.9995],[87,7.9776],[88,7.9291],[89,7.8437],[90,7.7312],[91,7.5982],[93,7.2668],[95,6.8715],[96,6.6577],[97,6.4297],[99,5.9427],[104,4.63],[106,4.1288],[107,3.8966],[108,3.6756],[109,3.4766],[110,3.2903],[111,3.1117],[112,2.9515],[113,2.7989],[114,2.6571],[115,2.5349],[116,2.4225],[117,2.3217],[118,2.2297],[119,2.1515],[120,2.0841],[121,2.0238],[122,1.9813],[124,1.9093],[125,1.8821],[128,1.8312],[129,1.8227],[133,1.7597],[134,1.7402],[136,1.6905],[137,1.6592],[139,1.5846],[140,1.5442],[141,1.5002],[144,1.3576],[149,1.0918],[152,0.9436],[154,0.856],[157,0.737],[158,0.7007],[159,0.6681],[160,0.6337],[161,0.6039],[162,0.5766],[163,0.5482],[164,0.5232],[166,0.4789],[168,0.4429],[169,0.4271],[170,0.4139],[172,0.3941],[173,0.3911],[175,0.3883],[176,0.3921],[177,0.4002],[179,0.4194],[180,0.4324],[181,0.4464],[182,0.4656],[183,0.4802],[185,0.513],[186,0.5311],[187,0.5448],[188,0.564],[189,0.5869],[190,0.6085],[191,0.6254],[192,0.6491],[193,0.6672],[194,0.6909],[195,0.7174],[196,0.7371],[199,0.8208],[200,0.8501],[202,0.9028],[204,0.9656],[205,1.0029],[206,1.037],[207,1.0681],[208,1.1077],[209,1.1422],[211,1.218],[214,1.3377],[216,1.4218],[217,1.47],[220,1.6275],[222,1.7235],[223,1.758],[224,1.7841],[225,1.8034],[226,1.8098],[227,1.8096],[228,1.8027],[230,1.7728],[233,1.7066],[234,1.6805],[235,1.6499],[237,1.5803],[240,1.4674],[241,1.4318],[242,1.4031],[243,1.3779],[244,1.3616],[245,1.3529],[246,1.352],[247,1.3557],[248,1.3729],[249,1.3976],[250,1.4351],[251,1.4839],[252,1.5493],[253,1.6259],[254,1.7132],[256,1.9153],[258,2.1453],[259,2.2639],[261,2.5135],[264,2.8702],[266,3.0797],[268,3.2716],[271,3.5124],[273,3.669],[275,3.8046],[276,3.8782],[280,4.1311],[284,4.3339],[293,4.7159],[296,4.8641],[299,5.026],[311,5.7637],[316,6.0351],[320,6.2242],[354,6.6514]]
       }
       body_ctrl.Location.Y {
        type=MovieSceneScriptingFloatChannel
        default=0
        thinned_from=315
        keys [[0,-0.8091],[8,-1.6755],[9,-1.7309],[10,-1.7816],[11,-1.8266],[12,-1.8673],[14,-1.9379],[17,-2.0204],[20,-2.0797],[23,-2.1231],[28,-2.1715],[33,-2.1927],[39,-2.1936],[41,-2.1814],[44,-2.1405],[45,-2.1237],[48,-2.0513],[54,-1.8692],[61,-1.6288],[65,-1.4939],[68,-1.3993],[69,-1.3657],[74,-1.2216],[76,-1.1705],[77,-1.1471],[78,-1.126],[80,-1.0938],[81,-1.0796],[82,-1.0753],[83,-1.0674],[85,-1.0668],[86,-1.0721],[87,-1.0913],[88,-1.1288],[89,-1.1832],[90,-1.2561],[91,-1.3444],[92,-1.4405],[93,-1.5534],[94,-1.6699],[95,-1.7937],[99,-2.3079],[100,-2.4325],[101,-2.5485],[102,-2.6589],[103,-2.7587],[104,-2.8446],[105,-2.9179],[106,-2.9772],[107,-3.0133],[108,-3.0345],[109,-3.0369],[111,-3.0045],[113,-2.9502],[116,-2.8343],[117,-2.7798],[118,-2.7169],[120,-2.6014],[123,-2.4227],[125,-2.3191],[126,-2.2796],[127,-2.2458],[129,-2.204],[130,-2.1937],[131,-2.1878],[132,-2.1915],[133,-2.2046],[134,-2.2225],[135,-2.2484],[136,-2.284],[137,-2.3251],[138,-2.3767],[140,-2.5023],[142,-2.6495],[144,-2.817],[147,-3.0952],[148,-3.1834],[149,-3.2814],[151,-3.4597],[152,-3.5541],[155,-3.8082],[159,-4.1085],[163,-4.3605],[166,-4.5132],[169,-4.6328],[171,-4.6975],[175,-4.7823],[183,-4.8563],[190,-4.8426],[194,-4.8135],[196,-4.7913],[197,-4.7706],[198,-4.7622],[202,-4.6795],[205,-4.5977],[207,-4.5314],[210,-4.4073],[214,-4.2059],[217,-4.0171],[219,-3.8742],[220,-3.7868],[221,-3.6896],[222,-3.5844],[224,-3.3542],[225,-3.2318],[228,-2.8367],[231,-2.4597],[234,-2.1052],[235,-1.9936],[237,-1.7852],[239,-1.6051],[240,-1.5243],[241,-1.4494],[242,-1.3874],[243,-1.3348],[244,-1.294],[245,-1.2628],[246,-1.2272],[247,-1.1995],[248,-1.1854],[249,-1.1551],[252,-1.0962],[253,-1.0798],[254,-1.0593],[256,-1.0102],[257,-0.9795],[260,-0.907],[261,-0.8813],[262,-0.8583],[263,-0.8322],[264,-0.8201],[265,-0.7984],[267,-0.764],[268,-0.749],[269,-0.736],[270,-0.7146],[271,-0.7024],[273,-0.6666],[274,-0.6411],[275,-0.6175],[276,-0.5884],[277,-0.5653],[279,-0.5129],[280,-0.4913],[281,-0.4666],[282,-0.4463],[284,-0.3998],[285,-0.3809],[286,-0.358],[287,-0.3395],[288,-0.3173],[290,-0.2806],[291,-0.2657],[292,-0.2498],[293,-0.2347],[294,-0.2179],[295,-0.2031],[298,-0.1661],[299,-0.1543],[300,-0.1442],[301,-0.1338],[302,-0.1211],[303,-0.1142],[305,-0.0901],[306,-0.079],[307,-0.069],[308,-0.0566],[309,-0.046],[310,-0.0331],[311,-0.0219],[312,-0.0082],[313,0.0042],[314,0.017],[315,0.0319],[316,0.0438],[317,0.0584],[318,0.0734],[319,0.0849],[320,0.1024],[354,-0.8091]]
       }
       body_ctrl.Location.Z {
        type=MovieSceneScriptingFloatChannel
        default=0
        keys [[0,38.2052],[54,4.9853],[129,50.0283],[205,11.709],[282,59.0968],[301,68.8174],[354,38.2052]]
       }
       body_ctrl.Rotation.X {
        type=MovieSceneScriptingFloatChannel
        default=0
        thinned_from=281
        keys [[0,-0.6782],[42,-0.7703],[43,-0.7786],[45,-0.8017],[48,-0.844],[50,-0.8809],[51,-0.9043],[53,-0.9622],[54,-0.9966],[56,-1.0729],[58,-1.1623],[60,-1.26],[63,-1.4205],[64,-1.4776],[65,-1.5316],[67,-1.6451],[70,-1.8013],[71,-1.8503],[73,-1.939],[76,-2.0433],[77,-2.0713],[82,-2.1833],[85,-2.2347],[86,-2.2561],[87,-2.2698],[88,-2.2783],[89,-2.2775],[90,-2.2678],[91,-2.2443],[92,-2.2132],[93,-2.1752],[94,-2.1255],[95,-2.0692],[96,-2.0026],[97,-1.9205],[99,-1.7463],[103,-1.3856],[106,-1.0923],[108,-0.9051],[109,-0.8201],[111,-0.6677],[112,-0.599],[113,-0.5385],[114,-0.4805],[115,-0.4299],[116,-0.3731],[117,-0.3037],[118,-0.2187],[119,-0.127],[120,-0.0285],[121,0.0777],[123,0.2865],[124,0.3891],[125,0.4866],[126,0.5773],[127,0.6604],[128,0.7333],[129,0.7992],[130,0.8546],[131,0.8986],[132,0.9349],[133,0.9534],[134,0.9649],[135,0.9632],[136,0.9527],[137,0.9244],[138,0.8832],[139,0.826],[140,0.7608],[141,0.6924],[142,0.6217],[143,0.5528],[144,0.4805],[146,0.3395],[147,0.2718],[148,0.2008],[149,0.1309],[150,0.0641],[151,-0.0032],[152,-0.0712],[153,-0.1373],[154,-0.1986],[155,-0.2591],[156,-0.313],[157,-0.3616],[158,-0.4084],[159,-0.4518],[160,-0.4911],[161,-0.5262],[162,-0.5567],[163,-0.5819],[164,-0.6003],[166,-0.6168],[167,-0.6109],[168,-0.5958],[169,-0.5764],[170,-0.5456],[171,-0.5052],[172,-0.4593],[173,-0.404],[174,-0.3397],[175,-0.2718],[176,-0.1974],[177,-0.1185],[178,-0.0332],[179,0.0545],[180,0.1443],[181,0.2369],[182,0.3302],[183,0.4246],[184,0.5168],[185,0.6127],[186,0.7048],[188,0.8925],[190,1.0736],[191,1.1603],[193,1.3259],[195,1.4773],[197,1.623],[199,1.76],[203,2.0183],[206,2.2286],[208,2.3886],[210,2.5698],[211,2.6683],[212,2.773],[214,3.0067],[215,3.1372],[216,3.2819],[217,3.4427],[218,3.6229],[219,3.8306],[220,4.0666],[221,4.3271],[225,5.4467],[226,5.7128],[227,5.9672],[228,6.2058],[229,6.4261],[230,6.6229],[231,6.8033],[232,6.9583],[233,7.0909],[234,7.196],[235,7.2767],[236,7.3304],[237,7.361],[238,7.3608],[239,7.3374],[240,7.2778],[241,7.1789],[242,7.038],[243,6.8513],[244,6.6213],[246,6.0989],[251,4.7472],[253,4.2288],[254,3.9822],[255,3.7444],[257,3.2979],[259,2.8776],[261,2.4789],[262,2.2903],[263,2.1127],[264,1.9458],[265,1.7925],[266,1.6485],[267,1.5105],[268,1.381],[269,1.2611],[270,1.1521],[271,1.0478],[272,0.9559],[273,0.8692],[274,0.7933],[275,0.7283],[276,0.6741],[277,0.6269],[280,0.4807],[281,0.4285],[282,0.3772],[283,0.3287],[284,0.2783],[285,0.2292],[288,0.0903],[289,0.0433],[290,-0.0021],[291,-0.0537],[292,-0.1048],[293,-0.1572],[295,-0.266],[296,-0.3179],[298,-0.43],[299,-0.4851],[300,-0.5426],[304,-0.7669],[306,-0.8753],[307,-0.9278],[308,-0.983],[311,-1.1342],[312,-1.1802],[314,-1.2762],[317,-1.4101],[320,-1.5362],[354,-0.6782]]
       }
       body_ctrl.Rotation.Y {
        type=MovieSceneScriptingFloatChannel
        default=0
        thinned_from=306
        keys [[0,-2.2517],[17,-4.5492],[20,-5.1735],[21,-5.3739],[22,-5.5619],[23,-5.7339],[24,-5.8837],[25,-6.0043],[26,-6.0897],[28,-6.204],[30,-6.2704],[32,-6.2963],[35,-6.2819],[41,-6.1633],[56,-5.769],[78,-4.9833],[83,-4.8364],[86,-4.736],[89,-4.6923],[90,-4.6911],[94,-4.7339],[95,-4.7276],[96,-4.7033],[97,-4.6514],[98,-4.5574],[99,-4.4144],[100,-4.231],[101,-4.0105],[102,-3.754],[103,-3.4618],[104,-3.2325],[105,-2.9789],[106,-2.7097],[107,-2.4289],[110,-1.5633],[111,-1.2774],[112,-1.002],[113,-0.7327],[114,-0.4735],[115,-0.226],[116,0.0149],[117,0.2393],[118,0.4472],[119,0.6334],[120,0.7975],[121,0.9389],[122,1.054],[123,1.1462],[124,1.2212],[125,1.2761],[126,1.3139],[127,1.332],[128,1.3363],[129,1.329],[130,1.3118],[135,1.1998],[137,1.1618],[141,1.0982],[146,1.032],[154,0.9326],[162,0.8209],[164,0.7953],[166,0.7747],[168,0.7575],[173,0.7253],[174,0.7154],[176,0.6917],[177,0.676],[178,0.6563],[181,0.5887],[182,0.5683],[184,0.5308],[185,0.5157],[186,0.5031],[189,0.4721],[191,0.4556],[193,0.4415],[194,0.4359],[195,0.4323],[196,0.4302],[197,0.427],[198,0.4259],[200,0.4273],[202,0.4332],[203,0.438],[204,0.4443],[206,0.4617],[207,0.4739],[208,0.4874],[209,0.5027],[210,0.5198],[211,0.542],[212,0.5653],[213,0.5919],[214,0.6235],[216,0.6992],[217,0.7333],[218,0.7655],[219,0.7957],[220,0.8329],[221,0.8823],[222,0.9491],[223,1.0424],[224,1.1575],[225,1.2897],[226,1.4343],[229,1.8869],[230,2.0255],[231,2.1496],[232,2.2504],[233,2.3244],[234,2.3716],[235,2.3999],[236,2.4036],[237,2.3909],[238,2.361],[239,2.3152],[240,2.2527],[241,2.1702],[242,2.0714],[243,1.9562],[244,1.8271],[247,1.4213],[249,1.1472],[250,1.0076],[251,0.8622],[252,0.7118],[253,0.5534],[254,0.3867],[255,0.2142],[256,0.033],[257,-0.152],[258,-0.3338],[259,-0.5144],[260,-0.6921],[261,-0.8604],[262,-1.0177],[263,-1.1588],[264,-1.2769],[265,-1.379],[266,-1.4567],[268,-1.5767],[269,-1.6077],[270,-1.6253],[278,-1.7214],[280,-1.7363],[289,-1.7483],[293,-1.7482],[295,-1.7589],[297,-1.7774],[298,-1.7911],[300,-1.8293],[302,-1.8814],[304,-1.9475],[306,-2.0256],[310,-2.1959],[312,-2.2678],[314,-2.3212],[315,-2.3399],[316,-2.3518],[317,-2.3558],[319,-2.3514],[320,-2.3438],[354,-2.2517]]
       }
       body_ctrl.Rotation.Z {
        type=MovieSceneScriptingFloatChannel
        default=0
        thinned_from=231
        keys [[0,-16.9771],[54,-15.8095],[77,-16.0246],[78,-16.0532],[90,-15.1532],[92,-14.5782],[93,-14.3437],[94,-14.1705],[95,-14.0739],[96,-14.0746],[97,-14.1909],[98,-14.4423],[99,-14.8315],[100,-15.3415],[101,-15.9569],[102,-16.6647],[103,-17.4463],[104,-18.2875],[109,-22.8715],[111,-24.6035],[112,-25.3987],[113,-26.1303],[115,-27.3889],[117,-28.5299],[118,-29.0328],[119,-29.4704],[120,-29.828],[121,-30.0902],[122,-30.2391],[124,-30.3143],[127,-30.202],[155,-27.7117],[176,-29.5075],[207,-32.6158],[217,-33.0298],[222,-32.8811],[225,-32.884],[228,-32.631],[230,-32.2617],[232,-31.6612],[233,-31.2617],[234,-30.7872],[235,-30.2328],[236,-29.606],[238,-28.5424],[241,-26.7245],[248,-22.1442],[250,-20.9162],[252,-19.788],[254,-18.789],[257,-17.4596],[259,-16.6691],[261,-15.9664],[263,-15.3577],[265,-14.8492],[267,-14.4526],[268,-14.301],[273,-13.8241],[278,-13.4966],[285,-13.2636],[293,-13.1985],[297,-13.3142],[320,-14.7686],[354,-16.9771]]
       }
       body_ctrl.Scale.X {type=MovieSceneScriptingFloatChannel, default=1, const [1,315]}
       body_ctrl.Scale.Y {type=MovieSceneScriptingFloatChannel, default=1, const [1,315]}
       body_ctrl.Scale.Z {type=MovieSceneScriptingFloatChannel, default=1, const [1,315]}
       hips_ctrl.Location.X {
        type=MovieSceneScriptingFloatChannel
        default=0
        thinned_from=322
        keys [[0,-0.1208],[1,-0.128],[4,-0.1334],[6,-0.1365],[7,-0.1394],[8,-0.1408],[9,-0.1436],[10,-0.1479],[11,-0.1535],[12,-0.1606],[13,-0.1692],[14,-0.1792],[15,-0.1906],[16,-0.2034],[17,-0.2212],[19,-0.2483],[20,-0.2683],[21,-0.2819],[23,-0.322],[28,-0.3897],[29,-0.3965],[30,-0.41],[32,-0.424],[33,-0.4265],[35,-0.4355],[36,-0.438],[37,-0.4425],[42,-0.4548],[43,-0.4552],[44,-0.4577],[45,-0.4581],[46,-0.4605],[51,-0.4614],[52,-0.4585],[53,-0.4587],[59,-0.4403],[60,-0.4342],[61,-0.4309],[62,-0.4247],[63,-0.4215],[64,-0.4153],[65,-0.4116],[77,-0.3525],[78,-0.3521],[80,-0.3427],[81,-0.3418],[82,-0.3418],[83,-0.3361],[87,-0.3344],[90,-0.3493],[94,-0.3926],[95,-0.3982],[96,-0.4091],[97,-0.415],[99,-0.4147],[101,-0.396],[102,-0.3822],[104,-0.3447],[106,-0.2978],[112,-0.1263],[113,-0.0997],[114,-0.067],[115,-0.0444],[116,-0.0229],[118,0.0223],[121,0.0564],[123,0.0588],[124,0.059],[125,0.0602],[127,0.0402],[128,0.0292],[129,0.0134],[130,0.0019],[131,-0.0099],[133,-0.0331],[134,-0.0445],[136,-0.056],[137,-0.0614],[138,-0.0614],[139,-0.0666],[140,-0.0664],[141,-0.0658],[142,-0.0598],[143,-0.0594],[144,-0.0588],[145,-0.0539],[147,-0.0458],[148,-0.0412],[151,-0.0291],[152,-0.0255],[154,-0.0174],[155,-0.0169],[158,-0.0046],[159,-0.0047],[160,-0.0006],[161,-0.0002],[162,0.0013],[163,0.0013],[164,0.0028],[170,0.0051],[171,0.0043],[172,0.0048],[173,0.0033],[174,0.0015],[175,-0.0008],[176,-0.0049],[177,-0.0105],[181,-0.034],[182,-0.0461],[186,-0.0711],[187,-0.0714],[188,-0.0781],[189,-0.0843],[190,-0.0909],[191,-0.0913],[192,-0.098],[193,-0.098],[194,-0.1024],[195,-0.1058],[196,-0.1068],[197,-0.1068],[198,-0.1112],[199,-0.1113],[200,-0.1123],[201,-0.1123],[202,-0.11],[203,-0.1099],[204,-0.1076],[205,-0.1076],[206,-0.1042],[207,-0.1019],[208,-0.0985],[209,-0.0913],[210,-0.0879],[211,-0.0804],[212,-0.077],[213,-0.0695],[214,-0.0584],[215,-0.0507],[216,-0.0393],[218,-0.0235],[220,-0.0153],[221,-0.007],[222,-0.0024],[223,0.0098],[224,0.0225],[225,0.0345],[226,0.0512],[227,0.0682],[228,0.0849],[230,0.1177],[231,0.1335],[233,0.1635],[234,0.1781],[240,0.2614],[241,0.2773],[246,0.3423],[247,0.3583],[250,0.3984],[251,0.4089],[253,0.4247],[255,0.4345],[256,0.4368],[258,0.4351],[260,0.4314],[262,0.4259],[263,0.4241],[264,0.4203],[266,0.4165],[267,0.4155],[269,0.4153],[270,0.4171],[271,0.417],[272,0.4159],[273,0.4157],[275,0.4095],[276,0.4089],[282,0.3895],[283,0.3889],[288,0.3745],[289,0.3704],[293,0.3698],[294,0.3626],[296,0.3631],[297,0.3559],[299,0.3569],[300,0.3497],[301,0.3502],[304,0.3295],[309,0.3015],[310,0.29],[315,0.2612],[316,0.2551],[317,0.2555],[318,0.2494],[319,0.2494],[320,0.2433],[354,-0.1208]]
       }
       hips_ctrl.Location.Y {
        type=MovieSceneScriptingFloatChannel
        default=0
        thinned_from=322
        keys [[0,-0.9889],[1,-0.977],[4,-0.9677],[12,-0.9578],[16,-0.9472],[30,-0.9251],[46,-0.8808],[47,-0.8808],[53,-0.8718],[57,-0.8712],[77,-0.9038],[78,-0.9064],[79,-0.9057],[81,-0.9102],[85,-0.915],[89,-0.9312],[92,-0.9386],[93,-0.9382],[94,-0.9398],[95,-0.9386],[96,-0.9401],[98,-0.939],[100,-0.9423],[103,-0.9525],[105,-0.9649],[108,-0.9927],[112,-1.0409],[114,-1.062],[115,-1.0746],[116,-1.0795],[118,-1.1048],[123,-1.139],[124,-1.139],[125,-1.1467],[127,-1.1494],[128,-1.1431],[129,-1.1449],[133,-1.1379],[136,-1.1366],[137,-1.1389],[138,-1.1389],[140,-1.1441],[144,-1.158],[145,-1.1588],[147,-1.1652],[148,-1.172],[151,-1.1815],[152,-1.181],[158,-1.2006],[159,-1.2006],[162,-1.2084],[163,-1.2084],[172,-1.2349],[173,-1.2411],[174,-1.2443],[175,-1.2446],[177,-1.251],[181,-1.2542],[187,-1.2422],[188,-1.2374],[189,-1.2354],[192,-1.2229],[193,-1.2229],[194,-1.2158],[195,-1.2162],[196,-1.2086],[197,-1.2086],[198,-1.2015],[199,-1.2015],[200,-1.1939],[201,-1.1939],[202,-1.1858],[203,-1.1858],[204,-1.1777],[206,-1.1773],[207,-1.1692],[214,-1.1594],[216,-1.1605],[220,-1.1721],[223,-1.1929],[228,-1.249],[230,-1.2644],[231,-1.2672],[233,-1.2607],[234,-1.2538],[237,-1.2183],[241,-1.1612],[243,-1.1294],[244,-1.1108],[248,-1.0467],[251,-1.0076],[255,-0.9701],[256,-0.9642],[257,-0.9624],[263,-0.9183],[264,-0.9035],[269,-0.8661],[270,-0.8658],[273,-0.8444],[276,-0.8318],[277,-0.8234],[283,-0.7986],[284,-0.7989],[286,-0.7908],[287,-0.7911],[288,-0.7871],[290,-0.7861],[291,-0.7821],[294,-0.7832],[295,-0.7871],[297,-0.7881],[299,-0.7961],[300,-0.7971],[302,-0.806],[303,-0.807],[306,-0.8228],[307,-0.8263],[308,-0.8323],[309,-0.8357],[310,-0.8426],[315,-0.8598],[316,-0.8607],[317,-0.8632],[320,-0.865],[354,-0.9889]]
       }
       hips_ctrl.Location.Z {
        type=MovieSceneScriptingFloatChannel
        default=0
        thinned_from=322
        keys [[0,0.0566],[1,0.0558],[12,0.0533],[15,0.0521],[17,0.051],[18,0.0502],[19,0.0498],[21,0.0481],[22,0.0478],[23,0.0469],[24,0.0464],[25,0.0454],[26,0.0452],[28,0.0439],[30,0.0436],[33,0.0428],[46,0.0409],[47,0.0412],[49,0.0408],[50,0.041],[51,0.0408],[52,0.041],[53,0.0406],[59,0.0411],[62,0.0411],[63,0.0415],[69,0.0416],[70,0.0422],[72,0.0423],[73,0.0419],[77,0.0427],[78,0.0424],[79,0.0425],[80,0.0428],[81,0.0424],[82,0.0433],[83,0.0432],[84,0.0429],[85,0.0422],[86,0.0427],[87,0.0414],[88,0.0409],[89,0.0402],[90,0.0401],[92,0.0364],[93,0.0354],[94,0.0331],[96,0.0311],[97,0.0297],[98,0.0289],[99,0.0284],[100,0.0283],[101,0.0285],[103,0.0298],[104,0.0307],[106,0.0332],[109,0.0376],[110,0.0393],[112,0.0421],[113,0.0431],[114,0.0446],[116,0.0468],[117,0.0473],[118,0.0482],[119,0.0501],[120,0.051],[121,0.0509],[122,0.0524],[123,0.0542],[124,0.0545],[125,0.057],[126,0.0574],[127,0.0595],[128,0.0596],[129,0.0613],[130,0.0622],[131,0.0635],[132,0.0645],[133,0.0653],[134,0.0658],[135,0.0668],[137,0.0676],[138,0.0682],[139,0.0681],[140,0.0677],[141,0.0681],[144,0.0673],[145,0.0665],[147,0.0658],[148,0.0656],[149,0.066],[151,0.0649],[152,0.0641],[155,0.0629],[157,0.0626],[158,0.0617],[159,0.0612],[160,0.0613],[161,0.0609],[163,0.0603],[164,0.0604],[165,0.0602],[166,0.0603],[168,0.0601],[170,0.06],[172,0.0604],[173,0.0607],[176,0.0611],[177,0.0614],[178,0.0613],[179,0.0618],[180,0.0618],[181,0.0626],[182,0.0622],[187,0.0633],[188,0.0633],[189,0.0635],[190,0.0634],[191,0.064],[192,0.0637],[193,0.0647],[194,0.0639],[195,0.0647],[196,0.0641],[197,0.0649],[198,0.0639],[199,0.0647],[200,0.0638],[201,0.0644],[202,0.0645],[203,0.0642],[204,0.0641],[205,0.0646],[207,0.0641],[208,0.0646],[210,0.0645],[211,0.0648],[212,0.0645],[218,0.0648],[220,0.0643],[221,0.0643],[223,0.0637],[224,0.0632],[226,0.0616],[227,0.0612],[228,0.0601],[229,0.0594],[232,0.0564],[233,0.0558],[234,0.055],[235,0.0545],[236,0.0536],[237,0.053],[238,0.0523],[242,0.0481],[243,0.0474],[244,0.0461],[246,0.044],[247,0.0427],[248,0.0417],[249,0.0402],[253,0.036],[254,0.0354],[256,0.0336],[257,0.0333],[258,0.0328],[260,0.0316],[261,0.0312],[264,0.0296],[265,0.0294],[266,0.0289],[267,0.029],[270,0.0281],[271,0.0281],[272,0.0278],[276,0.0283],[277,0.0281],[281,0.0286],[282,0.0286],[283,0.0284],[285,0.0287],[286,0.0284],[287,0.0287],[288,0.0284],[289,0.0289],[293,0.0279],[294,0.0279],[295,0.0283],[296,0.0276],[297,0.0281],[298,0.0275],[299,0.0276],[300,0.0272],[301,0.0272],[302,0.0268],[304,0.0272],[309,0.0262],[311,0.0264],[314,0.026],[315,0.0256],[317,0.0254],[318,0.0251],[320,0.0249],[354,0.0566]]
       }
       hips_ctrl.Rotation.X {
        type=MovieSceneScriptingFloatChannel
        default=0
        thinned_from=261
        keys [[0,3.66],[6,3.6158],[9,3.6128],[11,3.6318],[15,3.7257],[19,3.8477],[22,3.9636],[25,4.1108],[32,4.5051],[38,4.762],[42,4.8908],[46,4.9805],[49,5.0184],[52,5.0137],[55,4.9595],[58,4.8613],[62,4.6777],[72,4.1246],[75,3.9836],[78,3.8721],[81,3.7865],[83,3.75],[84,3.7213],[85,3.6746],[86,3.6154],[87,3.5363],[88,3.4426],[90,3.2242],[93,2.8745],[94,2.7644],[95,2.6625],[96,2.5687],[103,1.9779],[105,1.7934],[106,1.6965],[108,1.5129],[109,1.4293],[111,1.2791],[112,1.2129],[113,1.1555],[114,1.1032],[115,1.0599],[116,1.0341],[127,0.9257],[179,2.1851],[180,2.2365],[182,2.3538],[185,2.5583],[189,2.8421],[191,2.973],[193,3.0901],[195,3.1896],[198,3.3122],[201,3.4068],[204,3.4724],[207,3.5086],[209,3.5172],[211,3.5006],[213,3.4576],[214,3.4272],[215,3.3898],[216,3.3432],[220,3.111],[221,3.0453],[222,2.9686],[223,2.8513],[224,2.706],[225,2.5414],[226,2.3672],[228,2.01],[229,1.8398],[230,1.6837],[231,1.5415],[232,1.4234],[233,1.3291],[234,1.2579],[235,1.1937],[238,1.0189],[240,0.8895],[241,0.8281],[242,0.7704],[243,0.7158],[244,0.6597],[245,0.5996],[246,0.5349],[247,0.4635],[248,0.387],[250,0.2249],[251,0.1448],[252,0.0651],[254,-0.0925],[255,-0.1716],[256,-0.2486],[257,-0.3219],[258,-0.3888],[259,-0.4462],[260,-0.495],[261,-0.5345],[262,-0.5644],[263,-0.5877],[264,-0.6035],[265,-0.6135],[266,-0.6172],[267,-0.6144],[268,-0.6051],[269,-0.5897],[270,-0.569],[271,-0.5454],[272,-0.5167],[273,-0.4843],[274,-0.4477],[275,-0.4078],[277,-0.3137],[278,-0.2697],[279,-0.2293],[280,-0.192],[281,-0.1568],[282,-0.1251],[283,-0.0993],[284,-0.076],[285,-0.0571],[286,-0.0427],[287,-0.0323],[288,-0.0244],[289,-0.0215],[290,-0.0274],[291,-0.0361],[292,-0.0532],[293,-0.0759],[294,-0.104],[295,-0.1405],[296,-0.1853],[297,-0.2328],[298,-0.2857],[299,-0.3439],[300,-0.4046],[301,-0.4706],[302,-0.5392],[303,-0.6102],[304,-0.6837],[305,-0.7588],[306,-0.8318],[307,-0.908],[310,-1.1257],[312,-1.2568],[313,-1.3154],[314,-1.3705],[315,-1.4188],[316,-1.4638],[317,-1.502],[318,-1.5369],[319,-1.5682],[320,-1.5962],[354,3.66]]
       }
       hips_ctrl.Rotation.Y {
        type=MovieSceneScriptingFloatChannel
        default=0
        thinned_from=265
        keys [[0,0.5304],[58,0.2234],[59,0.221],[60,0.217],[62,0.2079],[65,0.1899],[67,0.177],[68,0.1699],[69,0.1638],[70,0.1573],[72,0.146],[74,0.1372],[75,0.1346],[76,0.1317],[77,0.1291],[78,0.1283],[79,0.1271],[80,0.1277],[81,0.1296],[83,0.143],[84,0.1531],[85,0.1727],[86,0.2069],[87,0.2617],[88,0.3369],[89,0.4266],[90,0.5416],[91,0.6791],[92,0.8386],[93,1.0089],[96,1.5643],[97,1.7445],[98,1.9124],[99,2.0615],[100,2.2005],[101,2.3289],[102,2.445],[103,2.5465],[104,2.6381],[105,2.707],[106,2.7594],[107,2.7949],[108,2.8178],[109,2.8278],[110,2.8288],[112,2.8073],[115,2.735],[117,2.679],[118,2.6564],[119,2.6419],[120,2.6348],[121,2.6379],[122,2.6523],[123,2.6771],[124,2.7089],[125,2.7495],[126,2.7972],[131,3.093],[133,3.1957],[135,3.2773],[136,3.3084],[138,3.3504],[140,3.3699],[143,3.373],[148,3.335],[157,3.2307],[161,3.2186],[172,3.3058],[174,3.3346],[176,3.376],[178,3.435],[182,3.578],[184,3.6386],[189,3.7341],[193,3.7741],[197,3.7765],[200,3.7532],[203,3.7066],[206,3.6336],[209,3.5322],[211,3.4459],[213,3.3427],[214,3.2814],[216,3.1417],[217,3.078],[218,3.0205],[219,2.9716],[223,2.8107],[226,2.6973],[227,2.6511],[228,2.5964],[229,2.5263],[230,2.4404],[231,2.3346],[232,2.206],[233,2.0483],[234,1.8743],[236,1.508],[237,1.3204],[239,0.9351],[240,0.7395],[241,0.5493],[242,0.3612],[243,0.1775],[244,-0.0014],[245,-0.1824],[247,-0.5387],[248,-0.7093],[249,-0.8751],[250,-1.0269],[251,-1.1649],[252,-1.296],[253,-1.4155],[254,-1.5189],[255,-1.613],[256,-1.691],[257,-1.7624],[260,-1.9915],[262,-2.1547],[264,-2.3017],[266,-2.4139],[267,-2.4615],[269,-2.5436],[270,-2.5778],[273,-2.6507],[275,-2.6814],[286,-2.751],[293,-2.769],[296,-2.7565],[298,-2.7357],[300,-2.6955],[302,-2.6359],[304,-2.5574],[306,-2.4611],[310,-2.2407],[312,-2.1366],[314,-2.0462],[315,-2.0078],[316,-1.9747],[318,-1.9271],[320,-1.8905],[354,0.5304]]
       }
       hips_ctrl.Rotation.Z {
        type=MovieSceneScriptingFloatChannel
        default=0
        thinned_from=311
        keys [[0,7.8157],[88,7.7821],[125,8.0146],[137,8.1798],[217,8.1008],[236,7.8019],[320,7.8539],[354,7.8157]]
       }
       hips_ctrl.Scale.X {type=MovieSceneScriptingFloatChannel, default=1, const [1,322]}
       hips_ctrl.Scale.Y {type=MovieSceneScriptingFloatChannel, default=1, const [1,322]}
       hips_ctrl.Scale.Z {type=MovieSceneScriptingFloatChannel, default=1, const [1,322]}
       hips_tan_ctrl.Location.X {
        type=MovieSceneScriptingFloatChannel
        default=0
        thinned_from=322
        keys [[0,-0.0581],[1,-0.0381],[2,-0.0361],[3,-0.026],[4,-0.0166],[5,-0.0077],[6,0.0008],[7,0.0079],[8,0.0154],[9,0.0223],[10,0.0306],[11,0.0419],[12,0.0558],[13,0.0656],[14,0.0848],[15,0.1062],[16,0.1212],[17,0.1431],[18,0.1669],[19,0.1803],[20,0.2027],[21,0.2168],[23,0.2534],[24,0.2634],[25,0.2771],[26,0.2816],[27,0.2904],[28,0.2917],[29,0.2964],[30,0.3039],[31,0.3051],[32,0.3031],[34,0.3054],[36,0.3025],[37,0.3031],[40,0.2969],[41,0.2938],[47,0.2795],[48,0.278],[49,0.2778],[50,0.273],[53,0.2677],[54,0.2653],[55,0.265],[57,0.2605],[58,0.2592],[59,0.2537],[60,0.2531],[62,0.248],[63,0.2441],[65,0.2395],[66,0.2402],[67,0.234],[68,0.2315],[69,0.2297],[71,0.225],[72,0.222],[75,0.2159],[76,0.2145],[77,0.2083],[78,0.2097],[79,0.2083],[80,0.2065],[81,0.2059],[82,0.2038],[83,0.1991],[85,0.1954],[86,0.1962],[87,0.1943],[88,0.1877],[89,0.188],[90,0.1833],[91,0.1755],[92,0.1698],[93,0.1704],[94,0.1624],[95,0.1509],[96,0.1376],[97,0.1303],[98,0.1116],[99,0.0963],[100,0.0783],[101,0.0634],[102,0.0294],[103,0.0113],[104,-0.0192],[105,-0.0396],[106,-0.0712],[107,-0.0865],[108,-0.1185],[109,-0.143],[110,-0.1608],[111,-0.1837],[112,-0.2052],[113,-0.2244],[114,-0.2421],[115,-0.2585],[116,-0.2625],[117,-0.2867],[118,-0.2985],[119,-0.2968],[120,-0.3159],[121,-0.3108],[122,-0.3254],[123,-0.3269],[124,-0.3169],[125,-0.3269],[127,-0.3223],[128,-0.3186],[129,-0.3156],[130,-0.3118],[133,-0.3042],[136,-0.299],[138,-0.2975],[139,-0.2982],[141,-0.2973],[146,-0.2989],[152,-0.3021],[159,-0.3046],[160,-0.3057],[161,-0.3059],[167,-0.3135],[168,-0.3131],[173,-0.3173],[174,-0.3194],[177,-0.3203],[178,-0.322],[179,-0.3169],[180,-0.3188],[181,-0.318],[182,-0.3198],[183,-0.3156],[184,-0.3157],[185,-0.3171],[186,-0.3175],[188,-0.3171],[189,-0.3194],[190,-0.3138],[192,-0.3141],[193,-0.3159],[194,-0.3147],[196,-0.314],[197,-0.3126],[200,-0.3115],[201,-0.312],[202,-0.3118],[203,-0.3102],[204,-0.3108],[207,-0.309],[208,-0.3079],[209,-0.3096],[210,-0.3089],[211,-0.3044],[212,-0.3039],[214,-0.3066],[215,-0.3023],[216,-0.304],[217,-0.3066],[218,-0.3074],[219,-0.3038],[220,-0.3068],[221,-0.3116],[222,-0.3179],[223,-0.3267],[224,-0.3375],[226,-0.3642],[227,-0.3726],[229,-0.3993],[230,-0.4045],[231,-0.4122],[232,-0.4137],[233,-0.4142],[235,-0.4038],[236,-0.393],[237,-0.3877],[238,-0.3752],[239,-0.3687],[241,-0.3455],[242,-0.3326],[243,-0.3234],[244,-0.3095],[245,-0.3004],[247,-0.273],[248,-0.2646],[250,-0.2385],[251,-0.2297],[252,-0.2161],[255,-0.1845],[256,-0.173],[257,-0.1635],[258,-0.15],[259,-0.142],[260,-0.1298],[261,-0.1148],[262,-0.106],[263,-0.0945],[264,-0.0816],[265,-0.0728],[266,-0.0691],[267,-0.0635],[268,-0.0591],[269,-0.0562],[270,-0.0539],[271,-0.0492],[272,-0.0474],[275,-0.0409],[276,-0.039],[277,-0.0333],[279,-0.0297],[280,-0.0281],[281,-0.0261],[282,-0.0243],[283,-0.035],[284,-0.03],[285,-0.0285],[286,-0.0273],[287,-0.0258],[288,-0.0212],[289,-0.0201],[291,-0.0184],[293,-0.0173],[294,-0.0163],[295,-0.0161],[296,-0.0156],[297,-0.0155],[298,-0.0151],[299,-0.0103],[300,-0.0141],[301,-0.0135],[302,-0.0131],[304,-0.012],[305,-0.0126],[306,-0.0107],[307,-0.0129],[308,-0.0134],[309,-0.0116],[310,-0.0108],[311,-0.0138],[312,-0.0128],[313,-0.0127],[314,-0.0115],[315,-0.0112],[316,-0.0067],[317,-0.0072],[318,-0.0054],[319,-0.0042],[320,-0.005],[354,-0.0581]]
       }
       hips_tan_ctrl.Location.Y {
        type=MovieSceneScriptingFloatChannel
        default=0.008
        thinned_from=322
        keys [[0,-0.7297],[1,-0.734],[2,-0.7357],[3,-0.7443],[7,-0.7487],[11,-0.7478],[12,-0.753],[14,-0.7485],[15,-0.7511],[20,-0.7322],[22,-0.7228],[23,-0.7169],[24,-0.7132],[25,-0.7052],[26,-0.701],[30,-0.6792],[31,-0.6721],[32,-0.6713],[33,-0.6651],[34,-0.6613],[35,-0.6544],[36,-0.6587],[37,-0.6524],[38,-0.6498],[39,-0.6439],[40,-0.6419],[41,-0.6437],[42,-0.6421],[43,-0.6375],[44,-0.6364],[45,-0.6319],[46,-0.6384],[47,-0.6378],[48,-0.6341],[50,-0.6359],[51,-0.6351],[52,-0.6359],[56,-0.6465],[60,-0.6634],[66,-0.693],[67,-0.6993],[71,-0.7176],[72,-0.7258],[73,-0.7277],[76,-0.7383],[77,-0.7446],[78,-0.7454],[84,-0.7578],[85,-0.7636],[86,-0.7674],[88,-0.7824],[89,-0.7932],[91,-0.8195],[92,-0.8312],[94,-0.8618],[99,-0.921],[100,-0.9271],[101,-0.9363],[103,-0.9511],[104,-0.9536],[105,-0.9593],[106,-0.9608],[107,-0.9645],[110,-0.961],[112,-0.9546],[113,-0.9497],[114,-0.9415],[115,-0.9389],[117,-0.9236],[118,-0.9121],[121,-0.8671],[122,-0.8462],[123,-0.8313],[126,-0.77],[127,-0.7462],[128,-0.7309],[130,-0.6981],[132,-0.6715],[135,-0.6371],[136,-0.6343],[137,-0.6262],[138,-0.6248],[140,-0.6274],[141,-0.6351],[143,-0.6403],[144,-0.6475],[145,-0.6527],[146,-0.6626],[147,-0.6666],[148,-0.6767],[149,-0.681],[150,-0.6916],[151,-0.6944],[152,-0.705],[153,-0.709],[154,-0.7189],[155,-0.7225],[156,-0.7319],[157,-0.7345],[158,-0.7392],[159,-0.7477],[161,-0.7531],[162,-0.7605],[163,-0.7619],[164,-0.7593],[165,-0.7608],[166,-0.7661],[168,-0.7683],[169,-0.7643],[170,-0.7696],[173,-0.7689],[174,-0.7638],[176,-0.7612],[177,-0.7565],[178,-0.7551],[179,-0.7492],[181,-0.744],[182,-0.7371],[185,-0.7281],[186,-0.7192],[187,-0.7148],[188,-0.7119],[189,-0.7076],[190,-0.6993],[191,-0.7004],[192,-0.6921],[194,-0.6862],[195,-0.6807],[196,-0.6808],[197,-0.6758],[198,-0.6745],[199,-0.6698],[200,-0.6706],[201,-0.6662],[202,-0.6674],[204,-0.6603],[205,-0.6619],[207,-0.6612],[208,-0.658],[211,-0.6573],[213,-0.6603],[216,-0.6615],[217,-0.6752],[218,-0.6742],[219,-0.6755],[220,-0.6731],[221,-0.6825],[222,-0.6798],[223,-0.6739],[224,-0.6831],[226,-0.6895],[227,-0.6979],[228,-0.7021],[230,-0.7202],[231,-0.7321],[232,-0.748],[233,-0.7542],[234,-0.7645],[237,-0.8037],[238,-0.8178],[239,-0.8352],[240,-0.8422],[241,-0.8611],[243,-0.8895],[244,-0.9083],[246,-0.9372],[247,-0.956],[248,-0.9711],[249,-0.9891],[250,-1.0026],[252,-1.0346],[253,-1.0482],[254,-1.0594],[257,-1.099],[260,-1.1317],[263,-1.1578],[266,-1.1742],[268,-1.1828],[269,-1.1821],[271,-1.1875],[272,-1.1851],[274,-1.1846],[275,-1.189],[278,-1.1818],[279,-1.1631],[280,-1.1667],[284,-1.1627],[285,-1.168],[291,-1.1703],[293,-1.1766],[295,-1.1806],[296,-1.1863],[297,-1.1883],[298,-1.1948],[299,-1.1975],[301,-1.21],[307,-1.2417],[317,-1.2875],[320,-1.2963],[354,-0.7297]]
       }
       hips_tan_ctrl.Location.Z {
        type=MovieSceneScriptingFloatChannel
        default=-0.0073
        thinned_from=322
        keys [[0,-0.0348],[1,-0.034],[2,-0.0338],[3,-0.0338],[4,-0.0341],[6,-0.034],[7,-0.0341],[8,-0.0338],[9,-0.034],[10,-0.0343],[11,-0.0305],[13,-0.0313],[14,-0.0275],[15,-0.0283],[16,-0.0247],[17,-0.0241],[18,-0.0206],[19,-0.0195],[20,-0.0144],[21,-0.0117],[22,-0.0097],[23,-0.0058],[24,-0.0045],[25,-0.0006],[26,0.0006],[27,0.0045],[28,0.0091],[29,0.0089],[31,0.0108],[32,0.014],[34,0.0139],[35,0.015],[36,0.0159],[37,0.0157],[38,0.0166],[39,0.0167],[40,0.0175],[41,0.0175],[42,0.0161],[45,0.0161],[46,0.016],[47,0.0151],[48,0.0151],[49,0.0161],[50,0.013],[52,0.0132],[53,0.012],[54,0.0126],[56,0.0127],[57,0.01],[58,0.01],[59,0.0091],[60,0.0092],[62,0.0037],[63,0.0057],[64,0.0031],[65,0.0032],[66,0.0027],[67,-0.0011],[69,-0.002],[70,-0.0019],[71,-0.0057],[72,-0.0062],[73,-0.0061],[74,-0.0066],[75,-0.0066],[76,-0.0071],[77,-0.0109],[78,-0.0109],[79,-0.0114],[80,-0.0114],[81,-0.01],[82,-0.01],[83,-0.0148],[84,-0.0136],[85,-0.014],[86,-0.0127],[87,-0.0161],[88,-0.015],[89,-0.018],[90,-0.0165],[91,-0.0194],[92,-0.0177],[93,-0.0205],[94,-0.0218],[95,-0.0202],[96,-0.022],[97,-0.0234],[98,-0.0234],[99,-0.0275],[100,-0.028],[101,-0.0363],[102,-0.0374],[103,-0.0377],[104,-0.0467],[105,-0.0469],[106,-0.0515],[107,-0.0517],[109,-0.0604],[110,-0.0561],[111,-0.0598],[112,-0.0596],[113,-0.0601],[114,-0.0592],[116,-0.058],[117,-0.0618],[118,-0.0529],[119,-0.0529],[120,-0.0487],[121,-0.0486],[122,-0.0444],[123,-0.0412],[124,-0.0414],[125,-0.0373],[126,-0.0339],[127,-0.0342],[128,-0.0311],[129,-0.0298],[133,-0.0228],[135,-0.019],[136,-0.0192],[137,-0.015],[138,-0.0175],[140,-0.0178],[141,-0.0162],[142,-0.0203],[143,-0.0205],[144,-0.0187],[145,-0.0203],[146,-0.0234],[147,-0.0226],[148,-0.0246],[149,-0.0237],[150,-0.0268],[151,-0.0258],[152,-0.0289],[153,-0.028],[154,-0.0311],[155,-0.0301],[156,-0.0333],[157,-0.0321],[158,-0.0322],[159,-0.0353],[160,-0.0345],[161,-0.0355],[162,-0.0345],[163,-0.0363],[164,-0.0354],[165,-0.0353],[166,-0.0373],[167,-0.0373],[168,-0.0363],[170,-0.0367],[171,-0.0349],[172,-0.0361],[174,-0.0365],[175,-0.0348],[176,-0.036],[177,-0.0354],[178,-0.0331],[179,-0.0332],[180,-0.0308],[181,-0.0345],[182,-0.0343],[183,-0.0322],[184,-0.0318],[185,-0.0294],[186,-0.0324],[187,-0.0298],[188,-0.0294],[189,-0.0268],[190,-0.0266],[191,-0.03],[192,-0.0272],[193,-0.0265],[194,-0.0262],[195,-0.0264],[196,-0.0255],[197,-0.0261],[198,-0.025],[199,-0.0255],[200,-0.0242],[201,-0.024],[202,-0.0245],[205,-0.024],[206,-0.0246],[207,-0.0234],[209,-0.0221],[210,-0.0244],[212,-0.024],[213,-0.0233],[215,-0.0228],[216,-0.0222],[217,-0.0248],[218,-0.0242],[220,-0.0238],[222,-0.0227],[223,-0.0219],[224,-0.0234],[225,-0.0216],[226,-0.026],[227,-0.0223],[228,-0.021],[229,-0.0259],[230,-0.022],[231,-0.0214],[232,-0.0263],[233,-0.0233],[234,-0.0291],[235,-0.0266],[236,-0.0324],[237,-0.0302],[238,-0.0337],[239,-0.034],[240,-0.037],[241,-0.0369],[242,-0.0438],[243,-0.0411],[244,-0.048],[245,-0.0451],[246,-0.0517],[247,-0.0488],[248,-0.0537],[249,-0.0522],[250,-0.049],[251,-0.0556],[252,-0.0526],[253,-0.0592],[254,-0.0564],[255,-0.0631],[256,-0.0604],[257,-0.0612],[258,-0.0667],[260,-0.0631],[261,-0.0686],[262,-0.0668],[263,-0.0665],[264,-0.0642],[265,-0.0712],[266,-0.069],[267,-0.0686],[270,-0.0668],[272,-0.0658],[274,-0.0655],[275,-0.0648],[276,-0.0665],[277,-0.0668],[278,-0.0668],[279,-0.0581],[280,-0.0592],[281,-0.0593],[282,-0.0592],[284,-0.0594],[285,-0.0585],[286,-0.0585],[287,-0.0601],[288,-0.0601],[289,-0.0608],[290,-0.0608],[293,-0.0604],[294,-0.0624],[295,-0.061],[296,-0.0608],[297,-0.0616],[298,-0.0614],[299,-0.0622],[300,-0.0621],[301,-0.0628],[302,-0.0647],[303,-0.0633],[304,-0.0642],[305,-0.0644],[307,-0.0652],[308,-0.0663],[309,-0.0662],[310,-0.0663],[311,-0.0668],[312,-0.0669],[313,-0.0674],[316,-0.0678],[317,-0.0678],[320,-0.0684],[354,-0.0348]]
       }
       hips_tan_ctrl.Rotation.X {type=MovieSceneScriptingFloatChannel, default=0, const [0,322]}
       hips_tan_ctrl.Rotation.Y {type=MovieSceneScriptingFloatChannel, default=0, const [0,322]}
       hips_tan_ctrl.Rotation.Z {type=MovieSceneScriptingFloatChannel, default=0, const [0,322]}
       hips_tan_ctrl.Scale.X {type=MovieSceneScriptingFloatChannel, default=1, const [1,322]}
       hips_tan_ctrl.Scale.Y {type=MovieSceneScriptingFloatChannel, default=1, const [1,322]}
       hips_tan_ctrl.Scale.Z {type=MovieSceneScriptingFloatChannel, default=1, const [1,322]}
       spine_01_ctrl.Location.X {
        type=MovieSceneScriptingFloatChannel
        default=0
        thinned_from=318
        keys [[0,-0.0419],[5,-0.0461],[12,-0.0461],[13,-0.0504],[15,-0.0504],[16,-0.0546],[17,-0.0556],[18,-0.0575],[19,-0.0614],[21,-0.0652],[22,-0.069],[28,-0.0806],[29,-0.0806],[30,-0.0825],[31,-0.0825],[32,-0.0844],[33,-0.084],[36,-0.085],[37,-0.085],[38,-0.0853],[40,-0.0853],[41,-0.0856],[47,-0.0856],[49,-0.0851],[54,-0.0851],[55,-0.0845],[57,-0.0845],[58,-0.0839],[59,-0.0839],[60,-0.0833],[62,-0.0833],[63,-0.0827],[64,-0.0827],[65,-0.083],[66,-0.0813],[71,-0.0813],[72,-0.0795],[77,-0.0795],[78,-0.0778],[80,-0.0778],[81,-0.0773],[87,-0.0773],[88,-0.0812],[89,-0.0812],[92,-0.0927],[93,-0.0927],[96,-0.1043],[97,-0.1052],[98,-0.1081],[100,-0.1081],[101,-0.1052],[102,-0.1052],[106,-0.0939],[107,-0.0882],[108,-0.0854],[109,-0.0797],[110,-0.0769],[112,-0.0655],[113,-0.0617],[116,-0.0445],[117,-0.0445],[119,-0.033],[120,-0.033],[121,-0.0273],[123,-0.0273],[124,-0.0216],[128,-0.0216],[129,-0.0221],[130,-0.0221],[132,-0.0232],[133,-0.0232],[134,-0.0237],[138,-0.0237],[139,-0.0243],[143,-0.0243],[144,-0.0248],[146,-0.0247],[147,-0.0249],[150,-0.0249],[151,-0.0252],[155,-0.0252],[156,-0.0255],[159,-0.0255],[160,-0.0257],[161,-0.0253],[162,-0.0263],[173,-0.0263],[174,-0.0272],[176,-0.0272],[177,-0.0277],[178,-0.0288],[179,-0.0288],[180,-0.0299],[181,-0.0299],[183,-0.032],[184,-0.032],[185,-0.0331],[188,-0.0331],[189,-0.0342],[192,-0.0342],[193,-0.0347],[202,-0.0347],[203,-0.0341],[204,-0.0341],[205,-0.0336],[206,-0.0336],[208,-0.0325],[210,-0.0324],[211,-0.0293],[214,-0.0293],[215,-0.0261],[218,-0.0261],[219,-0.023],[226,-0.023],[227,-0.0204],[229,-0.0204],[230,-0.0178],[231,-0.0178],[232,-0.0153],[233,-0.0153],[234,-0.0127],[235,-0.0127],[236,-0.0101],[237,-0.0101],[238,-0.0076],[239,-0.0076],[240,-0.005],[241,-0.0036],[242,-0.0012],[243,-0.0012],[244,0.0011],[245,0.0011],[247,0.0059],[248,0.0059],[250,0.0106],[251,0.0106],[252,0.013],[256,0.013],[257,0.0115],[258,0.0115],[259,0.0082],[261,0.0082],[262,0.0049],[265,0.0049],[266,0.0016],[272,0.0016],[273,0.0026],[282,0.0026],[283,-0.0041],[288,-0.0041],[289,-0.0036],[290,-0.0036],[291,-0.006],[294,-0.006],[295,-0.0085],[296,-0.0085],[297,-0.0109],[299,-0.0109],[301,-0.0157],[302,-0.0157],[304,-0.0206],[305,-0.0216],[306,-0.0231],[307,-0.0262],[310,-0.0308],[311,-0.0339],[315,-0.0401],[316,-0.0401],[318,-0.0432],[319,-0.0432],[320,-0.0447],[354,-0.0419]]
       }
       spine_01_ctrl.Location.Y {
        type=MovieSceneScriptingFloatChannel
        default=0
        thinned_from=318
        keys [[0,-0.1505],[5,-0.1544],[11,-0.1545],[12,-0.1584],[14,-0.1585],[15,-0.1624],[16,-0.1625],[17,-0.1637],[23,-0.177],[26,-0.1804],[30,-0.1805],[31,-0.1795],[32,-0.1796],[33,-0.1791],[34,-0.1791],[35,-0.1772],[36,-0.1772],[37,-0.1754],[38,-0.1754],[39,-0.1735],[40,-0.1735],[41,-0.1716],[42,-0.1716],[43,-0.1697],[44,-0.1697],[45,-0.1678],[47,-0.1678],[48,-0.1659],[49,-0.1659],[50,-0.1651],[53,-0.1641],[58,-0.1636],[61,-0.1642],[62,-0.1642],[65,-0.1649],[66,-0.1649],[67,-0.1668],[71,-0.1667],[72,-0.1688],[76,-0.1688],[77,-0.1707],[80,-0.1707],[84,-0.1703],[85,-0.1721],[86,-0.1721],[87,-0.1737],[88,-0.1737],[91,-0.1788],[92,-0.1788],[94,-0.1821],[98,-0.1817],[102,-0.1792],[105,-0.179],[107,-0.1798],[109,-0.1818],[110,-0.1833],[111,-0.1843],[113,-0.1874],[114,-0.1874],[115,-0.19],[116,-0.19],[117,-0.1926],[121,-0.1929],[122,-0.1904],[123,-0.1906],[126,-0.183],[127,-0.1778],[129,-0.1726],[135,-0.1534],[136,-0.1534],[137,-0.1502],[140,-0.1502],[141,-0.1534],[143,-0.1534],[144,-0.1566],[145,-0.1566],[146,-0.16],[147,-0.16],[148,-0.1634],[149,-0.1634],[150,-0.1667],[151,-0.1667],[152,-0.1701],[153,-0.1701],[154,-0.1735],[155,-0.1735],[156,-0.1769],[158,-0.1769],[159,-0.1803],[161,-0.1806],[162,-0.1829],[165,-0.1829],[166,-0.1853],[169,-0.1853],[170,-0.1876],[176,-0.1877],[177,-0.1868],[178,-0.1868],[179,-0.1844],[181,-0.1844],[182,-0.182],[183,-0.1821],[184,-0.1796],[185,-0.1797],[186,-0.1772],[187,-0.1773],[188,-0.1748],[189,-0.1749],[190,-0.1725],[191,-0.1725],[192,-0.1701],[193,-0.1698],[195,-0.1669],[196,-0.167],[197,-0.1655],[198,-0.1656],[199,-0.1641],[200,-0.1642],[201,-0.1627],[203,-0.1627],[204,-0.1613],[207,-0.1613],[208,-0.1599],[216,-0.1599],[217,-0.1667],[220,-0.1667],[221,-0.1735],[223,-0.1735],[224,-0.1803],[232,-0.2175],[239,-0.2172],[240,-0.2124],[241,-0.2136],[243,-0.2106],[244,-0.2104],[246,-0.2073],[247,-0.2073],[248,-0.2057],[249,-0.2057],[250,-0.2043],[253,-0.204],[254,-0.2025],[256,-0.2024],[257,-0.2036],[263,-0.2034],[264,-0.201],[268,-0.2009],[269,-0.1985],[271,-0.1985],[272,-0.1961],[278,-0.1961],[279,-0.187],[288,-0.187],[291,-0.1866],[292,-0.1878],[295,-0.1878],[296,-0.189],[297,-0.1889],[298,-0.1901],[299,-0.1901],[301,-0.1924],[302,-0.1924],[305,-0.1957],[307,-0.1969],[308,-0.1982],[316,-0.2032],[317,-0.2032],[318,-0.2038],[319,-0.2038],[320,-0.2044],[354,-0.1505]]
       }
       spine_01_ctrl.Location.Z {
        type=MovieSceneScriptingFloatChannel
        default=0
        thinned_from=318
        keys [[0,0.0028],[10,0.0026],[11,0.005],[13,0.0049],[14,0.0072],[15,0.0071],[16,0.0094],[17,0.0098],[20,0.0148],[21,0.0181],[29,0.0315],[30,0.0315],[32,0.0349],[33,0.0349],[38,0.0373],[39,0.0373],[40,0.0378],[46,0.038],[47,0.0375],[48,0.0376],[49,0.038],[50,0.0365],[53,0.0366],[54,0.035],[56,0.0351],[57,0.0335],[58,0.0336],[59,0.032],[60,0.032],[62,0.029],[63,0.029],[64,0.0275],[66,0.0275],[67,0.0233],[70,0.0233],[71,0.0192],[76,0.0191],[77,0.0149],[80,0.0149],[81,0.0154],[82,0.0154],[83,0.0131],[86,0.013],[87,0.0107],[88,0.0107],[89,0.0084],[90,0.0083],[91,0.006],[92,0.006],[94,0.0014],[95,0.0014],[96,-0.0009],[97,-0.0022],[100,-0.0088],[101,-0.0132],[103,-0.0176],[104,-0.022],[109,-0.0331],[110,-0.0331],[111,-0.0354],[112,-0.0354],[113,-0.0359],[117,-0.0361],[118,-0.0318],[119,-0.0318],[120,-0.0275],[121,-0.0275],[123,-0.0189],[124,-0.0188],[126,-0.0101],[127,-0.0099],[128,-0.0055],[129,-0.0037],[135,0.0102],[136,0.0102],[137,0.0125],[141,0.0124],[142,0.0102],[144,0.0101],[145,0.0096],[146,0.0074],[147,0.0074],[148,0.0053],[149,0.0053],[150,0.0031],[151,0.0031],[152,0.0009],[153,0.0009],[154,-0.0013],[155,-0.0013],[156,-0.0035],[158,-0.0035],[159,-0.0057],[160,-0.0057],[161,-0.0063],[162,-0.0064],[163,-0.0074],[165,-0.0074],[166,-0.0085],[170,-0.0086],[171,-0.0075],[174,-0.0075],[175,-0.0065],[176,-0.0065],[177,-0.0059],[178,-0.0045],[179,-0.0044],[180,-0.0029],[182,-0.0028],[183,-0.0014],[184,-0.0013],[185,0.0001],[186,0.0002],[187,0.0017],[188,0.0017],[189,0.0032],[191,0.0033],[192,0.0048],[193,0.0048],[194,0.0053],[195,0.0053],[196,0.0058],[197,0.0058],[198,0.0063],[199,0.0063],[200,0.0067],[206,0.0068],[207,0.0073],[208,0.0073],[209,0.0078],[210,0.0064],[216,0.0064],[217,0.0049],[220,0.0049],[221,0.0047],[223,0.0047],[224,0.0031],[225,0.0034],[226,0.0001],[228,-0.0002],[229,-0.0035],[231,-0.0038],[232,-0.0071],[233,-0.0071],[234,-0.0102],[235,-0.0102],[236,-0.0134],[237,-0.0134],[238,-0.0165],[239,-0.0165],[240,-0.0195],[241,-0.0204],[242,-0.0247],[243,-0.0247],[244,-0.029],[245,-0.029],[246,-0.0333],[247,-0.0333],[248,-0.0376],[250,-0.0375],[251,-0.0419],[252,-0.0419],[253,-0.0462],[254,-0.0462],[255,-0.0505],[256,-0.0505],[257,-0.0515],[258,-0.0554],[260,-0.0554],[261,-0.0594],[264,-0.0593],[265,-0.0633],[272,-0.0631],[273,-0.0627],[278,-0.0627],[279,-0.058],[288,-0.058],[289,-0.0591],[293,-0.0591],[294,-0.0603],[296,-0.0603],[297,-0.0614],[298,-0.0615],[299,-0.0626],[300,-0.0627],[302,-0.065],[303,-0.065],[304,-0.0662],[307,-0.068],[308,-0.0692],[315,-0.0735],[316,-0.0735],[318,-0.0747],[319,-0.0747],[320,-0.0753],[354,0.0028]]
       }
       spine_01_ctrl.Rotation.X {
        type=MovieSceneScriptingFloatChannel
        default=0
        thinned_from=307
        keys [[0,6.082],[7,6.2285],[17,5.9367],[19,5.7563],[25,5.1317],[27,4.9528],[29,4.7936],[31,4.6599],[33,4.5536],[35,4.4783],[38,4.4026],[41,4.3629],[45,4.3551],[49,4.3839],[51,4.4204],[53,4.4797],[56,4.6044],[60,4.8296],[70,5.5092],[74,5.7374],[78,5.9034],[84,6.0798],[86,6.1817],[87,6.2564],[88,6.3589],[89,6.4839],[90,6.6276],[92,6.9552],[99,8.2072],[101,8.5308],[103,8.8152],[104,8.9357],[105,9.0375],[106,9.1166],[107,9.1708],[108,9.1984],[109,9.1974],[111,9.1392],[113,9.0329],[115,8.8817],[116,8.7742],[117,8.6324],[118,8.4609],[119,8.2635],[120,8.0487],[122,7.5778],[126,6.5922],[128,6.1286],[129,5.919],[130,5.7258],[131,5.5509],[132,5.3953],[133,5.2579],[134,5.14],[135,5.0436],[136,4.9697],[137,4.9196],[138,4.8941],[139,4.8945],[140,4.9142],[142,4.9848],[144,5.0887],[147,5.2949],[154,5.8437],[158,6.0981],[161,6.2404],[163,6.3097],[166,6.3612],[169,6.3743],[172,6.3323],[175,6.246],[183,5.9015],[190,5.5399],[193,5.4182],[198,5.2867],[204,5.2007],[209,5.177],[212,5.1986],[218,5.3076],[219,5.3138],[220,5.3084],[223,5.2463],[225,5.2314],[227,5.2405],[228,5.2563],[229,5.2888],[230,5.3393],[231,5.4092],[232,5.502],[233,5.6183],[234,5.7583],[235,5.9207],[236,6.0992],[238,6.4964],[240,6.9414],[243,7.6589],[249,9.1842],[252,9.8888],[254,10.3097],[256,10.6925],[259,11.1888],[262,11.607],[265,11.9277],[268,12.1388],[271,12.2501],[274,12.2892],[286,12.202],[292,12.2983],[300,12.6662],[314,13.5104],[320,13.7349],[354,6.082]]
       }
       spine_01_ctrl.Rotation.Y {
        type=MovieSceneScriptingFloatChannel
        default=0
        thinned_from=303
        keys [[0,0.6232],[20,4.7351],[21,5.0366],[22,5.3178],[23,5.572],[24,5.7942],[25,5.9775],[26,6.1151],[27,6.2214],[28,6.3126],[29,6.3892],[31,6.495],[33,6.5445],[35,6.5512],[38,6.5164],[57,5.9831],[76,5.277],[84,5.0329],[86,5.0006],[89,5.0196],[92,5.0739],[93,5.0848],[94,5.0789],[95,5.0488],[96,4.9876],[97,4.8872],[98,4.7375],[99,4.5307],[100,4.2749],[101,3.9784],[102,3.6423],[103,3.2707],[104,2.8716],[105,2.4436],[106,1.9962],[108,1.0769],[109,0.6234],[110,0.1802],[111,-0.2452],[112,-0.6514],[113,-1.0351],[114,-1.3925],[115,-1.7312],[116,-2.0527],[117,-2.3481],[118,-2.6145],[119,-2.8471],[120,-3.0444],[121,-3.2028],[122,-3.3205],[123,-3.4048],[124,-3.4631],[125,-3.4955],[126,-3.5008],[127,-3.484],[128,-3.4497],[132,-3.2642],[133,-3.2278],[135,-3.177],[137,-3.1453],[141,-3.1138],[158,-3.1171],[162,-3.1416],[171,-3.2588],[174,-3.2822],[176,-3.2818],[178,-3.2597],[184,-3.1248],[186,-3.099],[199,-3.0231],[205,-3.0229],[209,-3.0451],[213,-3.0934],[215,-3.1337],[218,-3.2139],[219,-3.2471],[220,-3.2957],[221,-3.3659],[222,-3.4641],[223,-3.5997],[224,-3.7656],[225,-3.9532],[228,-4.5493],[229,-4.729],[230,-4.8832],[231,-5.0029],[232,-5.0774],[233,-5.0972],[234,-5.0726],[235,-5.0199],[236,-4.9463],[238,-4.7721],[240,-4.568],[241,-4.4566],[245,-3.9526],[251,-3.194],[252,-3.0583],[253,-2.9136],[254,-2.762],[256,-2.4406],[257,-2.2704],[258,-2.0878],[259,-1.8973],[261,-1.4908],[262,-1.2903],[263,-1.1014],[264,-0.9356],[265,-0.793],[266,-0.6776],[267,-0.5875],[268,-0.5189],[269,-0.4776],[270,-0.448],[271,-0.4202],[272,-0.3941],[273,-0.366],[276,-0.2724],[277,-0.2389],[278,-0.2114],[279,-0.1819],[280,-0.1544],[283,-0.0662],[287,0.0432],[288,0.0685],[289,0.0955],[290,0.1207],[292,0.1697],[293,0.1948],[294,0.2213],[296,0.2769],[297,0.3054],[298,0.3366],[299,0.3733],[300,0.4146],[301,0.4596],[302,0.5068],[303,0.5565],[309,0.866],[317,1.3012],[319,1.4021],[320,1.4487],[354,0.6232]]
       }
       spine_01_ctrl.Rotation.Z {
        type=MovieSceneScriptingFloatChannel
        default=0
        thinned_from=318
        keys [[0,0.4973],[5,0.3926],[6,0.3484],[7,0.2982],[8,0.2431],[9,0.1837],[11,0.0579],[12,-0.0038],[13,-0.0662],[14,-0.128],[16,-0.2545],[17,-0.3183],[19,-0.4484],[22,-0.6461],[23,-0.7096],[25,-0.829],[26,-0.8829],[27,-0.9319],[28,-0.9746],[29,-1.01],[30,-1.0371],[31,-1.058],[32,-1.0723],[33,-1.0826],[47,-1.1537],[81,-1.2863],[83,-1.2922],[84,-1.3076],[85,-1.3422],[86,-1.3999],[87,-1.486],[88,-1.5981],[89,-1.73],[90,-1.8854],[91,-2.0665],[92,-2.2649],[94,-2.6825],[95,-2.8862],[96,-3.0754],[97,-3.2439],[98,-3.3829],[99,-3.4868],[100,-3.5612],[101,-3.6098],[102,-3.6333],[103,-3.6318],[104,-3.6051],[105,-3.5529],[106,-3.475],[107,-3.3789],[108,-3.2714],[110,-3.0349],[113,-2.668],[116,-2.331],[117,-2.2239],[118,-2.1236],[119,-2.037],[120,-1.9702],[121,-1.9261],[122,-1.9083],[123,-1.9125],[124,-1.9316],[125,-1.9687],[126,-2.0248],[127,-2.0962],[128,-2.1784],[131,-2.4442],[132,-2.5241],[133,-2.5966],[134,-2.6596],[135,-2.7152],[136,-2.7607],[137,-2.7988],[138,-2.8276],[139,-2.8491],[141,-2.8746],[143,-2.8823],[146,-2.8687],[160,-2.7206],[170,-2.6792],[173,-2.6845],[175,-2.7069],[176,-2.723],[177,-2.7465],[178,-2.7785],[182,-2.9327],[184,-2.9949],[185,-3.0164],[189,-3.0513],[193,-3.0569],[196,-3.0375],[199,-2.9987],[202,-2.9374],[205,-2.8506],[207,-2.7786],[209,-2.6913],[211,-2.5914],[213,-2.4739],[214,-2.4064],[215,-2.3282],[216,-2.2356],[217,-2.1288],[218,-2.0111],[219,-1.8824],[220,-1.7336],[221,-1.5646],[222,-1.3646],[223,-1.1396],[224,-0.8929],[225,-0.6342],[226,-0.3615],[227,-0.083],[228,0.1995],[229,0.4782],[230,0.7499],[231,1.0098],[232,1.2504],[233,1.4671],[234,1.6586],[235,1.8148],[236,1.9481],[237,2.0631],[238,2.1601],[239,2.2434],[240,2.3104],[241,2.3628],[242,2.402],[243,2.428],[244,2.4422],[245,2.4459],[246,2.4403],[247,2.4257],[248,2.4025],[249,2.372],[250,2.3345],[251,2.291],[252,2.2418],[254,2.1301],[259,1.8203],[260,1.7631],[261,1.7099],[262,1.6612],[263,1.6184],[264,1.5828],[265,1.5538],[266,1.5337],[267,1.5226],[268,1.5216],[269,1.5308],[273,1.583],[278,1.6282],[287,1.683],[292,1.7032],[294,1.6992],[295,1.692],[297,1.6682],[298,1.6491],[299,1.6223],[300,1.5882],[301,1.5482],[302,1.5027],[303,1.4521],[304,1.3974],[306,1.2777],[309,1.0781],[312,0.8721],[314,0.7394],[315,0.6756],[316,0.6139],[317,0.5551],[318,0.5004],[319,0.4486],[320,0.4007],[354,0.4973]]
       }
       spine_01_ctrl.Scale.X {type=MovieSceneScriptingFloatChannel, default=1, const [1,318]}
       spine_01_ctrl.Scale.Y {type=MovieSceneScriptingFloatChannel, default=1, const [1,318]}
       spine_01_ctrl.Scale.Z {type=MovieSceneScriptingFloatChannel, default=1, const [1,318]}
       spine_02_ctrl.Location.X {
        type=MovieSceneScriptingFloatChannel
        default=0
        thinned_from=318
        keys [[0,-0.0821],[5,-0.0876],[10,-0.0896],[11,-0.0905],[12,-0.0919],[13,-0.0938],[14,-0.0961],[15,-0.0988],[17,-0.1048],[18,-0.1116],[19,-0.1116],[20,-0.1184],[21,-0.1184],[23,-0.132],[24,-0.132],[25,-0.1387],[26,-0.1387],[27,-0.1455],[29,-0.1455],[30,-0.1523],[32,-0.1523],[33,-0.1516],[34,-0.1531],[36,-0.1531],[37,-0.1546],[48,-0.1546],[49,-0.1561],[50,-0.1536],[58,-0.1536],[59,-0.151],[62,-0.151],[63,-0.1484],[64,-0.1484],[65,-0.1491],[66,-0.1491],[67,-0.1459],[71,-0.1459],[72,-0.1427],[76,-0.1427],[77,-0.1395],[80,-0.1395],[82,-0.1397],[83,-0.1375],[85,-0.1375],[86,-0.1397],[87,-0.1397],[89,-0.1441],[90,-0.1484],[91,-0.1506],[94,-0.1637],[96,-0.1681],[97,-0.1728],[101,-0.1728],[102,-0.1627],[103,-0.1627],[104,-0.1527],[105,-0.1527],[106,-0.1426],[107,-0.1426],[109,-0.1224],[110,-0.1224],[112,-0.1023],[113,-0.0959],[115,-0.0764],[116,-0.0764],[118,-0.057],[119,-0.057],[120,-0.0473],[121,-0.0473],[122,-0.0376],[124,-0.0376],[125,-0.0279],[131,-0.0279],[132,-0.0285],[133,-0.0285],[134,-0.0292],[135,-0.0292],[136,-0.0298],[140,-0.0298],[141,-0.0305],[143,-0.0305],[144,-0.0311],[146,-0.0311],[147,-0.0314],[148,-0.0314],[149,-0.0317],[150,-0.0317],[151,-0.032],[152,-0.032],[154,-0.0326],[155,-0.0326],[156,-0.0329],[157,-0.0329],[160,-0.0338],[161,-0.0337],[162,-0.035],[167,-0.035],[168,-0.0362],[176,-0.0362],[177,-0.0369],[178,-0.0369],[179,-0.0414],[182,-0.0414],[183,-0.0459],[189,-0.0459],[190,-0.0503],[192,-0.0503],[193,-0.0492],[194,-0.05],[196,-0.05],[197,-0.0508],[200,-0.0508],[201,-0.05],[203,-0.05],[205,-0.0483],[206,-0.0483],[208,-0.0467],[209,-0.0446],[210,-0.0446],[211,-0.0426],[212,-0.0426],[217,-0.033],[224,-0.033],[225,-0.032],[228,-0.032],[229,-0.0304],[231,-0.0304],[232,-0.0287],[233,-0.0287],[234,-0.0271],[236,-0.0271],[237,-0.0254],[238,-0.0254],[240,-0.0221],[241,-0.0217],[242,-0.0204],[244,-0.0187],[247,-0.0147],[248,-0.013],[250,-0.0104],[252,-0.0087],[253,-0.0087],[255,-0.0095],[256,-0.0108],[257,-0.0125],[259,-0.0171],[260,-0.0171],[262,-0.0217],[263,-0.0217],[264,-0.024],[265,-0.024],[266,-0.0263],[270,-0.0263],[271,-0.0285],[272,-0.0285],[273,-0.0273],[276,-0.0273],[277,-0.0303],[283,-0.0303],[284,-0.0332],[287,-0.0332],[288,-0.0362],[289,-0.0356],[290,-0.0356],[291,-0.0392],[294,-0.0392],[295,-0.0427],[296,-0.0427],[297,-0.0463],[298,-0.0463],[299,-0.0499],[300,-0.0499],[301,-0.0535],[302,-0.0535],[304,-0.0606],[305,-0.0612],[306,-0.0653],[308,-0.0693],[309,-0.0733],[316,-0.0874],[317,-0.0874],[318,-0.0894],[320,-0.0894],[354,-0.0821]]
       }
       spine_02_ctrl.Location.Y {
        type=MovieSceneScriptingFloatChannel
        default=0
        thinned_from=318
        keys [[0,-0.2992],[10,-0.3075],[12,-0.3102],[14,-0.3143],[18,-0.3257],[23,-0.3417],[24,-0.3419],[25,-0.345],[31,-0.3457],[32,-0.3426],[35,-0.3434],[36,-0.3373],[40,-0.3374],[41,-0.3312],[45,-0.3313],[46,-0.3252],[49,-0.3249],[50,-0.3232],[52,-0.3232],[53,-0.3215],[56,-0.3213],[57,-0.3196],[64,-0.3191],[66,-0.3194],[67,-0.3208],[72,-0.3205],[73,-0.3222],[77,-0.3219],[78,-0.3236],[84,-0.323],[87,-0.3253],[89,-0.3264],[91,-0.3263],[93,-0.3245],[97,-0.3183],[99,-0.3123],[100,-0.312],[103,-0.3029],[104,-0.3029],[105,-0.2998],[106,-0.2998],[107,-0.2966],[112,-0.2963],[114,-0.2992],[116,-0.2992],[117,-0.3012],[118,-0.3012],[120,-0.3049],[121,-0.3049],[122,-0.3069],[123,-0.3072],[124,-0.3089],[125,-0.3092],[126,-0.3112],[127,-0.3112],[128,-0.3098],[130,-0.3086],[131,-0.3067],[134,-0.3041],[135,-0.3043],[136,-0.3032],[137,-0.3032],[138,-0.3044],[139,-0.3044],[143,-0.3083],[144,-0.3102],[145,-0.3106],[150,-0.3165],[151,-0.3189],[157,-0.3261],[158,-0.3261],[160,-0.3285],[163,-0.3285],[164,-0.3321],[168,-0.3321],[169,-0.3358],[173,-0.3359],[174,-0.3397],[176,-0.3397],[177,-0.3407],[183,-0.3409],[184,-0.3366],[187,-0.3369],[188,-0.3326],[190,-0.3326],[191,-0.3286],[192,-0.3286],[193,-0.3277],[194,-0.3255],[195,-0.3255],[196,-0.3233],[197,-0.3233],[198,-0.3211],[199,-0.3211],[200,-0.3188],[201,-0.3188],[202,-0.3166],[204,-0.3166],[205,-0.3144],[206,-0.3144],[207,-0.3122],[211,-0.3119],[212,-0.3101],[215,-0.3101],[216,-0.3118],[217,-0.3118],[218,-0.3135],[219,-0.3135],[221,-0.3169],[222,-0.3169],[223,-0.3204],[224,-0.3221],[225,-0.3252],[226,-0.3274],[227,-0.3272],[228,-0.3294],[231,-0.3292],[232,-0.327],[233,-0.3269],[236,-0.3202],[237,-0.32],[240,-0.3132],[241,-0.3118],[242,-0.3092],[243,-0.3072],[244,-0.3045],[248,-0.2964],[250,-0.2936],[253,-0.2913],[254,-0.2912],[255,-0.2904],[257,-0.2917],[263,-0.2911],[264,-0.2893],[265,-0.2893],[266,-0.2875],[267,-0.2875],[272,-0.279],[274,-0.2785],[275,-0.2737],[279,-0.2738],[280,-0.2692],[284,-0.2692],[285,-0.2644],[288,-0.2646],[289,-0.2639],[293,-0.2639],[294,-0.2652],[296,-0.2651],[297,-0.2663],[298,-0.2663],[300,-0.2687],[301,-0.2686],[305,-0.2738],[309,-0.2773],[310,-0.2791],[316,-0.2844],[317,-0.2844],[318,-0.2853],[319,-0.2853],[320,-0.2862],[354,-0.2992]]
       }
       spine_02_ctrl.Location.Z {
        type=MovieSceneScriptingFloatChannel
        default=0
        thinned_from=318
        keys [[0,-0.0281],[5,-0.0293],[9,-0.0295],[12,-0.0289],[14,-0.028],[17,-0.0261],[18,-0.0263],[19,-0.0236],[20,-0.0238],[22,-0.0183],[23,-0.0185],[24,-0.0156],[25,-0.0158],[26,-0.0128],[28,-0.0128],[29,-0.0098],[30,-0.0098],[31,-0.0068],[32,-0.0066],[33,-0.0066],[34,-0.0054],[35,-0.0054],[36,-0.005],[37,-0.0038],[40,-0.0038],[41,-0.0033],[42,-0.0021],[45,-0.0021],[46,-0.0016],[53,-0.0013],[54,-0.0043],[58,-0.0041],[59,-0.0071],[62,-0.0071],[63,-0.0101],[66,-0.0101],[67,-0.0139],[70,-0.0139],[71,-0.0176],[76,-0.0177],[77,-0.0214],[80,-0.0215],[81,-0.0221],[83,-0.022],[84,-0.023],[85,-0.0231],[93,-0.0307],[94,-0.0325],[95,-0.0334],[96,-0.0351],[97,-0.0357],[98,-0.0391],[99,-0.0389],[100,-0.0425],[101,-0.0422],[103,-0.049],[104,-0.049],[105,-0.0524],[106,-0.0524],[107,-0.0557],[109,-0.0557],[110,-0.0593],[116,-0.0596],[117,-0.0561],[119,-0.0562],[120,-0.0527],[121,-0.0527],[123,-0.0455],[124,-0.0456],[126,-0.0385],[127,-0.0385],[128,-0.0347],[129,-0.0332],[135,-0.0218],[137,-0.0217],[138,-0.02],[140,-0.02],[141,-0.0219],[143,-0.0221],[144,-0.0241],[145,-0.0237],[147,-0.0256],[148,-0.0275],[157,-0.0364],[158,-0.0364],[160,-0.0383],[161,-0.0383],[162,-0.0393],[163,-0.0393],[164,-0.0405],[167,-0.0405],[168,-0.0415],[169,-0.0418],[171,-0.0418],[172,-0.0408],[173,-0.0408],[174,-0.0411],[175,-0.0411],[176,-0.0402],[180,-0.04],[181,-0.0371],[183,-0.0371],[184,-0.0367],[185,-0.0367],[186,-0.0338],[187,-0.0338],[188,-0.0334],[190,-0.0334],[191,-0.0301],[192,-0.0301],[193,-0.0304],[195,-0.0291],[196,-0.0289],[197,-0.0284],[198,-0.0282],[199,-0.0276],[201,-0.0274],[202,-0.0267],[204,-0.0267],[205,-0.0265],[206,-0.026],[212,-0.0255],[213,-0.0259],[215,-0.0259],[216,-0.0264],[217,-0.0264],[218,-0.027],[220,-0.0271],[221,-0.0277],[222,-0.0277],[223,-0.0279],[224,-0.0285],[226,-0.0289],[227,-0.0312],[229,-0.0314],[230,-0.0336],[232,-0.0334],[233,-0.0356],[234,-0.0355],[235,-0.0375],[236,-0.0374],[238,-0.0416],[239,-0.0415],[240,-0.0435],[253,-0.063],[256,-0.0678],[257,-0.0686],[262,-0.0757],[263,-0.0757],[264,-0.077],[265,-0.077],[266,-0.0783],[272,-0.0776],[273,-0.0769],[274,-0.0769],[275,-0.0765],[276,-0.0751],[279,-0.0751],[280,-0.0733],[284,-0.0733],[285,-0.0729],[286,-0.0729],[287,-0.0715],[288,-0.0715],[289,-0.0725],[294,-0.0726],[295,-0.0736],[296,-0.0736],[297,-0.0748],[298,-0.0748],[299,-0.076],[300,-0.0761],[301,-0.0771],[302,-0.0772],[304,-0.0796],[308,-0.0822],[309,-0.0834],[315,-0.0873],[316,-0.0874],[318,-0.0886],[319,-0.0886],[320,-0.0893],[354,-0.0281]]
       }
       spine_02_ctrl.Rotation.X {
        type=MovieSceneScriptingFloatChannel
        default=0
        thinned_from=265
        keys [[0,-4.6275],[58,-5.2409],[77,-5.1036],[85,-5.0786],[87,-5.0864],[89,-5.0706],[91,-5.0286],[93,-4.959],[96,-4.815],[99,-4.6369],[104,-4.2941],[106,-4.1767],[107,-4.1281],[108,-4.0879],[110,-4.0333],[112,-3.9967],[114,-3.9782],[116,-3.9797],[117,-3.9936],[118,-4.0162],[120,-4.0854],[122,-4.1809],[126,-4.3934],[129,-4.5172],[131,-4.5757],[133,-4.6164],[149,-4.7821],[166,-4.8326],[180,-5.0491],[184,-5.0691],[191,-5.0203],[212,-4.7428],[216,-4.7181],[224,-4.7691],[226,-4.7625],[227,-4.7475],[228,-4.7211],[230,-4.6393],[241,-4.0562],[246,-3.7511],[249,-3.5948],[251,-3.5114],[253,-3.4481],[255,-3.4047],[263,-3.3177],[266,-3.2633],[272,-3.1129],[275,-3.0548],[286,-2.9167],[293,-2.874],[298,-2.8774],[320,-3.0641],[354,-4.6275]]
       }
       spine_02_ctrl.Rotation.Y {
        type=MovieSceneScriptingFloatChannel
        default=0
        thinned_from=309
        keys [[0,1.9371],[8,2.0565],[15,2.3687],[17,2.5162],[22,2.9379],[23,3.0179],[24,3.0885],[25,3.1501],[26,3.2009],[28,3.2771],[30,3.3322],[32,3.3676],[37,3.4016],[51,3.3769],[58,3.3252],[83,3.0358],[85,3.0214],[86,3.0219],[87,3.0306],[89,3.0704],[90,3.0986],[94,3.2432],[95,3.2686],[96,3.283],[97,3.2841],[98,3.2662],[99,3.2278],[100,3.1721],[101,3.1038],[102,3.021],[103,2.9277],[104,2.8206],[105,2.7026],[106,2.5733],[107,2.4365],[109,2.1459],[112,1.6992],[113,1.5549],[114,1.4148],[115,1.2777],[116,1.1442],[117,1.0156],[118,0.8955],[119,0.7816],[120,0.6779],[121,0.5835],[122,0.5001],[123,0.427],[124,0.3623],[125,0.3079],[126,0.2634],[127,0.2288],[128,0.2026],[129,0.1835],[130,0.1713],[131,0.1624],[134,0.1404],[136,0.1294],[137,0.1248],[138,0.1218],[139,0.1213],[140,0.1214],[142,0.1268],[143,0.1299],[144,0.1344],[145,0.1395],[146,0.146],[147,0.1516],[148,0.1586],[149,0.1653],[150,0.1723],[151,0.179],[158,0.2214],[159,0.2264],[161,0.2383],[163,0.2514],[164,0.2571],[165,0.2599],[166,0.2618],[167,0.2628],[168,0.2629],[169,0.2619],[171,0.2582],[172,0.2552],[173,0.2536],[174,0.2543],[175,0.256],[176,0.2586],[177,0.265],[178,0.2745],[179,0.2868],[180,0.3005],[181,0.3157],[182,0.3297],[183,0.3411],[184,0.3548],[186,0.3763],[188,0.392],[189,0.4017],[190,0.4084],[191,0.4167],[193,0.4315],[195,0.4435],[197,0.4529],[199,0.4594],[201,0.4619],[202,0.462],[203,0.4604],[205,0.454],[206,0.4485],[207,0.4418],[208,0.4338],[209,0.4239],[210,0.4127],[211,0.3998],[212,0.3853],[213,0.369],[214,0.3492],[215,0.3267],[216,0.3098],[217,0.3035],[218,0.3082],[219,0.3226],[220,0.342],[221,0.3632],[222,0.3827],[223,0.3998],[224,0.4147],[228,0.4685],[229,0.4832],[230,0.4991],[231,0.5166],[232,0.5367],[235,0.6063],[236,0.6279],[237,0.6457],[238,0.6601],[239,0.6719],[241,0.6902],[242,0.698],[243,0.7072],[244,0.7143],[245,0.7168],[246,0.7159],[247,0.7122],[249,0.6982],[250,0.6935],[251,0.6927],[252,0.6989],[253,0.7105],[254,0.7303],[255,0.7575],[256,0.7924],[257,0.8332],[259,0.922],[262,1.0635],[263,1.1062],[264,1.1442],[265,1.1756],[266,1.2031],[267,1.2237],[268,1.2399],[269,1.2498],[273,1.2771],[275,1.2961],[280,1.3398],[281,1.3462],[283,1.3674],[284,1.3821],[285,1.394],[288,1.4386],[289,1.4575],[290,1.4726],[292,1.5101],[294,1.5537],[297,1.6281],[299,1.6862],[301,1.7529],[303,1.8245],[310,2.0979],[312,2.1668],[314,2.2251],[316,2.2706],[320,2.3253],[354,1.9371]]
       }
       spine_02_ctrl.Rotation.Z {
        type=MovieSceneScriptingFloatChannel
        default=0
        thinned_from=300
        keys [[0,0.0892],[5,-0.074],[6,-0.1197],[7,-0.1697],[8,-0.2244],[9,-0.2804],[10,-0.3417],[11,-0.4079],[12,-0.4776],[13,-0.5505],[14,-0.6266],[16,-0.7846],[17,-0.8659],[18,-0.9495],[20,-1.113],[22,-1.2703],[23,-1.3444],[24,-1.4155],[25,-1.4805],[26,-1.539],[27,-1.5927],[28,-1.6398],[29,-1.6781],[30,-1.7095],[31,-1.7339],[32,-1.7529],[33,-1.7668],[38,-1.8081],[46,-1.8455],[53,-1.8559],[59,-1.8456],[65,-1.8227],[73,-1.7858],[80,-1.7657],[83,-1.7626],[84,-1.7694],[85,-1.7957],[86,-1.8365],[87,-1.9028],[88,-1.9891],[89,-2.0894],[90,-2.212],[91,-2.3517],[92,-2.5077],[94,-2.8347],[95,-2.991],[96,-3.1376],[97,-3.2617],[98,-3.361],[99,-3.4275],[100,-3.4676],[101,-3.483],[102,-3.4761],[103,-3.4466],[104,-3.3938],[105,-3.3205],[106,-3.2232],[107,-3.1108],[109,-2.8626],[112,-2.4699],[113,-2.3442],[116,-1.9961],[117,-1.8862],[118,-1.7817],[119,-1.6892],[120,-1.6126],[121,-1.5565],[122,-1.5212],[123,-1.5048],[124,-1.5007],[125,-1.5106],[126,-1.538],[127,-1.5779],[128,-1.6299],[131,-1.8065],[132,-1.8606],[133,-1.9107],[134,-1.953],[136,-2.0224],[138,-2.0705],[139,-2.0862],[140,-2.0902],[141,-2.1103],[144,-2.1275],[147,-2.126],[172,-2.0493],[173,-2.0497],[174,-2.0546],[176,-2.0752],[177,-2.0907],[179,-2.1432],[182,-2.2346],[184,-2.2798],[185,-2.2965],[189,-2.3146],[193,-2.3133],[195,-2.3042],[205,-1.9744],[207,-1.8853],[208,-1.836],[209,-1.7804],[210,-1.7163],[211,-1.6413],[212,-1.556],[213,-1.4634],[214,-1.3597],[215,-1.2412],[216,-1.1017],[217,-0.9421],[218,-0.7572],[219,-0.5566],[220,-0.3468],[221,-0.1268],[222,0.0922],[223,0.3122],[224,0.531],[225,0.7441],[226,0.94],[227,1.1221],[228,1.2857],[229,1.4257],[230,1.541],[231,1.634],[232,1.7129],[233,1.7832],[234,1.837],[239,1.9294],[245,1.9405],[246,1.9273],[247,1.9064],[248,1.8805],[249,1.8491],[251,1.7718],[253,1.6752],[255,1.5659],[257,1.4489],[259,1.3375],[261,1.235],[263,1.1519],[264,1.1213],[265,1.0938],[266,1.0757],[267,1.0666],[268,1.0626],[271,1.0941],[272,1.1061],[276,1.1396],[281,1.1683],[282,1.1715],[283,1.1772],[285,1.1831],[292,1.2032],[294,1.1959],[296,1.1789],[297,1.1671],[298,1.1505],[299,1.1247],[300,1.0963],[301,1.0614],[302,1.0216],[303,0.9776],[304,0.9308],[306,0.8263],[308,0.7127],[312,0.4806],[314,0.3662],[315,0.3126],[316,0.2611],[317,0.2117],[318,0.1656],[319,0.1227],[320,0.0829],[354,0.0892]]
       }
       spine_02_ctrl.Scale.X {type=MovieSceneScriptingFloatChannel, default=1, const [1,318]}
       spine_02_ctrl.Scale.Y {type=MovieSceneScriptingFloatChannel, default=1, const [1,318]}
       spine_02_ctrl.Scale.Z {type=MovieSceneScriptingFloatChannel, default=1, const [1,318]}
       spine_03_ctrl.Location.X {
        type=MovieSceneScriptingFloatChannel
        default=0
        thinned_from=315
        keys [[0,-0.1341],[8,-0.1423],[11,-0.1423],[12,-0.145],[13,-0.145],[16,-0.1533],[17,-0.1551],[18,-0.16],[19,-0.16],[21,-0.1697],[22,-0.1697],[23,-0.1746],[24,-0.1746],[25,-0.1794],[26,-0.1794],[27,-0.1843],[30,-0.1843],[31,-0.1892],[33,-0.1891],[42,-0.1939],[43,-0.1939],[45,-0.195],[46,-0.195],[47,-0.1955],[48,-0.1955],[49,-0.1966],[54,-0.1966],[55,-0.1945],[59,-0.1945],[60,-0.1925],[62,-0.1925],[63,-0.1904],[64,-0.1904],[65,-0.1893],[66,-0.1893],[67,-0.1877],[68,-0.1877],[69,-0.186],[70,-0.186],[71,-0.1843],[73,-0.1843],[74,-0.1826],[75,-0.1826],[76,-0.1809],[78,-0.1809],[79,-0.1793],[80,-0.1793],[81,-0.1785],[83,-0.1785],[84,-0.1774],[86,-0.1774],[90,-0.1817],[94,-0.1903],[96,-0.1924],[97,-0.1941],[98,-0.1941],[102,-0.1845],[105,-0.1702],[107,-0.1558],[108,-0.151],[109,-0.1414],[112,-0.1198],[113,-0.113],[119,-0.0703],[123,-0.0513],[124,-0.0489],[125,-0.0441],[126,-0.0418],[127,-0.0418],[128,-0.0394],[129,-0.0386],[131,-0.0381],[133,-0.0381],[139,-0.0388],[141,-0.0395],[145,-0.0406],[150,-0.0429],[151,-0.0429],[155,-0.0447],[156,-0.0447],[160,-0.0465],[161,-0.0468],[162,-0.0475],[163,-0.0475],[164,-0.0482],[165,-0.0482],[166,-0.0489],[169,-0.0489],[170,-0.0496],[174,-0.0496],[175,-0.0503],[176,-0.0503],[177,-0.0511],[178,-0.0511],[180,-0.0556],[181,-0.0556],[182,-0.0578],[183,-0.0578],[184,-0.06],[185,-0.06],[186,-0.0622],[188,-0.0622],[189,-0.0644],[191,-0.0644],[192,-0.0666],[193,-0.0664],[204,-0.0664],[205,-0.0643],[207,-0.0643],[208,-0.0623],[210,-0.0602],[211,-0.0582],[212,-0.0571],[216,-0.0489],[218,-0.0468],[219,-0.0478],[220,-0.0478],[221,-0.0489],[224,-0.0489],[226,-0.0483],[227,-0.0478],[228,-0.0475],[229,-0.0471],[230,-0.0468],[232,-0.0468],[233,-0.0471],[236,-0.0485],[240,-0.0495],[241,-0.0501],[246,-0.0501],[247,-0.0492],[248,-0.0492],[249,-0.0483],[253,-0.0483],[254,-0.0492],[257,-0.0544],[261,-0.0622],[263,-0.0657],[264,-0.067],[265,-0.0682],[266,-0.0691],[267,-0.0698],[268,-0.0703],[273,-0.0709],[275,-0.0709],[276,-0.0718],[279,-0.0718],[280,-0.0727],[282,-0.0727],[283,-0.0736],[284,-0.0736],[286,-0.0754],[287,-0.0754],[289,-0.0771],[290,-0.0787],[291,-0.0787],[293,-0.0817],[294,-0.0817],[301,-0.0924],[302,-0.0955],[304,-0.0985],[305,-0.1013],[306,-0.1046],[307,-0.1046],[309,-0.1114],[310,-0.1114],[311,-0.1147],[312,-0.1147],[313,-0.1181],[314,-0.1181],[315,-0.1214],[318,-0.1214],[319,-0.1248],[320,-0.1248],[354,-0.1341]]
       }
       spine_03_ctrl.Location.Y {
        type=MovieSceneScriptingFloatChannel
        default=0
        thinned_from=315
        keys [[0,-0.411],[9,-0.414],[11,-0.4201],[12,-0.4202],[14,-0.4265],[15,-0.4266],[16,-0.4298],[17,-0.4309],[22,-0.4324],[23,-0.4308],[24,-0.4313],[26,-0.4287],[28,-0.4292],[29,-0.4277],[30,-0.4277],[31,-0.4266],[32,-0.4266],[35,-0.4236],[42,-0.4189],[47,-0.4156],[50,-0.4156],[51,-0.4123],[56,-0.4119],[57,-0.4086],[63,-0.4078],[64,-0.4045],[65,-0.4053],[67,-0.4035],[68,-0.4033],[70,-0.4014],[72,-0.4012],[86,-0.3995],[87,-0.3969],[88,-0.3968],[89,-0.3941],[90,-0.3889],[91,-0.3862],[96,-0.3598],[97,-0.3571],[98,-0.3517],[99,-0.3479],[100,-0.3424],[101,-0.3387],[102,-0.3331],[103,-0.3292],[104,-0.3238],[108,-0.3088],[112,-0.3014],[113,-0.2985],[114,-0.2986],[115,-0.2967],[117,-0.2969],[120,-0.3033],[121,-0.3075],[122,-0.3136],[123,-0.3176],[124,-0.3257],[125,-0.3318],[127,-0.3477],[128,-0.3576],[129,-0.3652],[134,-0.4093],[139,-0.4312],[143,-0.431],[144,-0.4266],[145,-0.4275],[147,-0.4235],[148,-0.4196],[157,-0.4017],[158,-0.4016],[159,-0.3996],[160,-0.3995],[161,-0.3975],[162,-0.3975],[163,-0.395],[169,-0.3949],[170,-0.3974],[171,-0.3974],[176,-0.4098],[177,-0.4115],[180,-0.4218],[181,-0.4218],[183,-0.4287],[184,-0.4289],[185,-0.4323],[187,-0.4324],[188,-0.4359],[197,-0.4362],[198,-0.4308],[202,-0.4309],[203,-0.4254],[206,-0.4254],[207,-0.42],[208,-0.4201],[209,-0.4192],[210,-0.4164],[211,-0.4164],[212,-0.4135],[213,-0.4135],[215,-0.4077],[216,-0.4077],[218,-0.4019],[219,-0.4019],[221,-0.3961],[224,-0.3789],[225,-0.3694],[226,-0.3576],[227,-0.3516],[228,-0.3397],[229,-0.3336],[230,-0.3217],[234,-0.2974],[235,-0.2973],[236,-0.2912],[237,-0.291],[238,-0.2848],[239,-0.2846],[240,-0.2784],[241,-0.2774],[243,-0.2692],[244,-0.2687],[246,-0.2604],[247,-0.2604],[248,-0.2566],[249,-0.2561],[250,-0.2522],[252,-0.2517],[253,-0.2474],[254,-0.2474],[255,-0.2469],[256,-0.2469],[257,-0.2481],[262,-0.2483],[264,-0.2473],[267,-0.2441],[272,-0.2362],[273,-0.2356],[274,-0.2335],[275,-0.2335],[276,-0.2315],[277,-0.2315],[278,-0.2295],[279,-0.2295],[280,-0.2274],[282,-0.2275],[283,-0.2254],[285,-0.2254],[286,-0.2233],[288,-0.2233],[289,-0.2223],[291,-0.2215],[295,-0.221],[299,-0.2219],[304,-0.2246],[305,-0.2245],[306,-0.2261],[307,-0.226],[308,-0.2277],[309,-0.2276],[310,-0.2294],[311,-0.2292],[312,-0.231],[313,-0.2308],[314,-0.2326],[316,-0.2324],[317,-0.2342],[319,-0.2341],[320,-0.2358],[354,-0.411]]
       }
       spine_03_ctrl.Location.Z {
        type=MovieSceneScriptingFloatChannel
        default=0
        thinned_from=315
        keys [[0,-0.0473],[9,-0.0477],[11,-0.0483],[12,-0.0473],[14,-0.046],[15,-0.045],[16,-0.0443],[17,-0.0425],[18,-0.0425],[20,-0.034],[21,-0.034],[22,-0.0297],[23,-0.0295],[24,-0.0252],[25,-0.0251],[26,-0.0206],[27,-0.0206],[28,-0.0163],[30,-0.0162],[31,-0.0117],[33,-0.0115],[35,-0.01],[37,-0.0093],[38,-0.0087],[39,-0.0084],[40,-0.0083],[41,-0.008],[45,-0.0077],[46,-0.0079],[47,-0.0078],[48,-0.008],[49,-0.0075],[50,-0.0075],[51,-0.0072],[52,-0.0072],[53,-0.0109],[56,-0.0109],[57,-0.0106],[58,-0.0106],[59,-0.0143],[62,-0.0143],[63,-0.0181],[64,-0.0177],[65,-0.0178],[66,-0.0195],[67,-0.0195],[69,-0.0231],[70,-0.023],[71,-0.0248],[73,-0.0247],[74,-0.0266],[75,-0.0266],[76,-0.0284],[78,-0.0284],[79,-0.0303],[80,-0.0303],[81,-0.0308],[82,-0.0308],[83,-0.0317],[84,-0.0317],[85,-0.0326],[86,-0.0326],[87,-0.0332],[88,-0.0341],[89,-0.0348],[90,-0.0351],[91,-0.0358],[92,-0.0361],[93,-0.0374],[95,-0.0381],[96,-0.0394],[99,-0.0441],[100,-0.0466],[101,-0.0482],[103,-0.0534],[104,-0.0548],[105,-0.0575],[106,-0.0592],[107,-0.0619],[108,-0.0625],[109,-0.0644],[110,-0.0652],[111,-0.065],[112,-0.0658],[113,-0.0659],[114,-0.065],[115,-0.0648],[117,-0.0629],[120,-0.058],[121,-0.0557],[122,-0.0535],[123,-0.0521],[124,-0.0501],[125,-0.048],[126,-0.0461],[127,-0.0451],[128,-0.0443],[130,-0.0416],[134,-0.0394],[137,-0.0386],[139,-0.0395],[140,-0.0395],[141,-0.0403],[142,-0.0403],[143,-0.041],[144,-0.0413],[145,-0.0422],[147,-0.0434],[148,-0.0438],[156,-0.0488],[157,-0.0486],[158,-0.0494],[159,-0.0501],[160,-0.0509],[161,-0.051],[162,-0.0513],[163,-0.0514],[164,-0.0518],[165,-0.0518],[166,-0.0521],[169,-0.0521],[170,-0.0524],[172,-0.0523],[173,-0.0526],[177,-0.0522],[178,-0.0526],[179,-0.0518],[180,-0.0522],[181,-0.0522],[182,-0.0514],[183,-0.0518],[184,-0.0507],[185,-0.0511],[186,-0.0499],[187,-0.0499],[188,-0.0492],[189,-0.0492],[190,-0.048],[192,-0.048],[193,-0.0473],[194,-0.0473],[195,-0.0461],[197,-0.0461],[198,-0.0455],[199,-0.0455],[200,-0.0442],[202,-0.0442],[203,-0.0436],[206,-0.0436],[207,-0.043],[208,-0.0418],[209,-0.0422],[210,-0.0417],[211,-0.0417],[212,-0.0414],[213,-0.0414],[215,-0.0409],[216,-0.0409],[218,-0.0403],[219,-0.0403],[221,-0.0391],[225,-0.0351],[226,-0.0328],[227,-0.0321],[228,-0.0308],[229,-0.0312],[230,-0.0299],[231,-0.0303],[232,-0.0296],[233,-0.0301],[234,-0.0315],[235,-0.0326],[236,-0.033],[237,-0.0351],[238,-0.0365],[239,-0.0386],[240,-0.0401],[241,-0.0413],[242,-0.0454],[243,-0.0449],[244,-0.0495],[245,-0.049],[246,-0.0531],[247,-0.0531],[248,-0.0527],[249,-0.0572],[250,-0.0568],[251,-0.0613],[252,-0.0613],[253,-0.0654],[254,-0.0654],[255,-0.0699],[256,-0.0699],[260,-0.0766],[262,-0.0793],[263,-0.0804],[264,-0.0812],[266,-0.0824],[268,-0.0827],[273,-0.0818],[274,-0.081],[275,-0.081],[276,-0.0801],[277,-0.0801],[278,-0.0793],[279,-0.0787],[280,-0.0785],[281,-0.0785],[282,-0.0779],[283,-0.0776],[284,-0.0776],[285,-0.0771],[286,-0.0768],[288,-0.0768],[289,-0.0766],[290,-0.0771],[292,-0.0771],[293,-0.0776],[294,-0.0776],[300,-0.0809],[301,-0.082],[303,-0.0832],[304,-0.0844],[305,-0.0844],[307,-0.0872],[308,-0.0874],[309,-0.0887],[310,-0.0889],[311,-0.0902],[312,-0.0904],[313,-0.0917],[315,-0.0919],[316,-0.0932],[318,-0.0934],[319,-0.0947],[320,-0.0949],[354,-0.0473]]
       }
       spine_03_ctrl.Rotation.X {
        type=MovieSceneScriptingFloatChannel
        default=0
        thinned_from=311
        keys [[0,-2.6387],[12,-3.1083],[18,-3.1633],[58,-3.0568],[87,-2.8926],[89,-2.863],[101,-2.5797],[108,-2.3849],[112,-2.2943],[114,-2.2568],[115,-2.2201],[116,-2.1908],[117,-2.1727],[118,-2.1662],[119,-2.1717],[120,-2.1891],[121,-2.2199],[122,-2.2641],[123,-2.32],[124,-2.3862],[125,-2.4647],[126,-2.554],[127,-2.6522],[131,-3.085],[132,-3.1875],[133,-3.2831],[134,-3.3718],[135,-3.4491],[136,-3.5148],[137,-3.5661],[138,-3.602],[139,-3.6212],[140,-3.6273],[142,-3.6167],[144,-3.5801],[146,-3.524],[152,-3.3257],[155,-3.2445],[161,-3.1285],[164,-3.086],[166,-3.0735],[169,-3.0811],[171,-3.1013],[174,-3.1568],[176,-3.2048],[184,-3.4505],[188,-3.5316],[192,-3.5792],[196,-3.5908],[203,-3.5508],[209,-3.478],[214,-3.38],[219,-3.2534],[220,-3.2194],[221,-3.1761],[222,-3.1201],[223,-3.0386],[224,-2.9378],[225,-2.8234],[228,-2.448],[229,-2.3307],[230,-2.2237],[231,-2.1282],[232,-2.0467],[233,-1.9836],[234,-1.9341],[235,-1.8933],[237,-1.8277],[245,-1.6079],[248,-1.5366],[249,-1.5157],[251,-1.4833],[252,-1.4729],[253,-1.4658],[255,-1.4613],[257,-1.4675],[259,-1.4816],[262,-1.5087],[263,-1.5148],[264,-1.5175],[266,-1.5136],[268,-1.4975],[271,-1.4601],[273,-1.441],[275,-1.4293],[295,-1.3744],[298,-1.3723],[302,-1.3814],[310,-1.4254],[320,-1.5027],[354,-2.6387]]
       }
       spine_03_ctrl.Rotation.Y {
        type=MovieSceneScriptingFloatChannel
        default=0
        thinned_from=310
        keys [[0,1.7435],[13,1.7812],[15,1.7997],[20,1.8601],[23,1.8865],[25,1.8953],[51,1.9105],[52,1.9088],[66,1.896],[84,1.8388],[88,1.8054],[92,1.7548],[96,1.683],[98,1.6392],[101,1.5589],[103,1.4973],[105,1.4271],[106,1.3886],[107,1.3471],[109,1.2556],[113,1.0539],[119,0.7435],[123,0.5297],[124,0.477],[125,0.4256],[126,0.3755],[127,0.3278],[128,0.2826],[129,0.2406],[130,0.2029],[131,0.1686],[132,0.1374],[133,0.1089],[134,0.0838],[135,0.0624],[136,0.0447],[137,0.0306],[138,0.0217],[139,0.0171],[140,0.0166],[141,0.0182],[142,0.0226],[143,0.0283],[144,0.0359],[145,0.045],[146,0.056],[147,0.0674],[148,0.0809],[149,0.0925],[150,0.1053],[151,0.1174],[152,0.1289],[153,0.1409],[154,0.1518],[155,0.1632],[157,0.1831],[158,0.1924],[161,0.2179],[163,0.233],[164,0.2391],[165,0.2437],[166,0.2469],[167,0.2485],[169,0.2489],[170,0.2476],[171,0.2453],[173,0.2383],[174,0.2356],[176,0.2278],[178,0.2226],[180,0.2198],[182,0.219],[183,0.218],[184,0.219],[186,0.2197],[188,0.2223],[190,0.226],[193,0.2347],[196,0.2452],[198,0.2514],[199,0.2551],[202,0.2634],[203,0.2658],[205,0.2689],[207,0.2705],[210,0.2692],[212,0.267],[213,0.2654],[214,0.2626],[215,0.2604],[216,0.2635],[217,0.2756],[218,0.2965],[219,0.3247],[220,0.3584],[221,0.395],[222,0.434],[224,0.5178],[230,0.7762],[232,0.8658],[234,0.9573],[236,1.0432],[238,1.1199],[240,1.1884],[243,1.2787],[244,1.3057],[246,1.3509],[248,1.3849],[252,1.4414],[254,1.4763],[256,1.518],[262,1.6651],[263,1.6877],[265,1.724],[267,1.751],[270,1.7772],[284,1.842],[291,1.8986],[308,2.1014],[312,2.1337],[316,2.1477],[320,2.1442],[354,1.7435]]
       }
       spine_03_ctrl.Rotation.Z {
        type=MovieSceneScriptingFloatChannel
        default=0
        thinned_from=297
        keys [[0,0.2767],[8,-0.0865],[10,-0.2338],[11,-0.3123],[13,-0.4798],[16,-0.746],[18,-0.9286],[21,-1.196],[22,-1.2819],[24,-1.4446],[26,-1.5918],[27,-1.6578],[28,-1.7135],[29,-1.761],[30,-1.8001],[32,-1.8556],[36,-1.9134],[44,-1.9927],[51,-2.0309],[58,-2.0325],[79,-1.9466],[83,-1.9424],[84,-1.9544],[85,-1.9859],[86,-2.0388],[87,-2.1221],[88,-2.2304],[89,-2.3599],[90,-2.5127],[91,-2.694],[92,-2.8914],[95,-3.5146],[96,-3.7041],[97,-3.8733],[98,-4.0091],[99,-4.1084],[100,-4.1763],[101,-4.2159],[102,-4.2271],[103,-4.2109],[104,-4.1674],[105,-4.0956],[106,-3.9964],[107,-3.8758],[108,-3.745],[113,-3.0238],[116,-2.6291],[117,-2.5021],[118,-2.3817],[119,-2.2748],[120,-2.1877],[121,-2.1235],[122,-2.0861],[123,-2.071],[124,-2.0707],[125,-2.0892],[126,-2.1268],[127,-2.1814],[128,-2.2487],[131,-2.4787],[132,-2.5499],[133,-2.6136],[134,-2.6701],[135,-2.7185],[136,-2.7609],[138,-2.8233],[139,-2.8473],[141,-2.8796],[143,-2.8982],[145,-2.906],[170,-2.818],[173,-2.8184],[176,-2.8513],[177,-2.872],[179,-2.9402],[181,-3.0201],[183,-3.0866],[184,-3.113],[185,-3.132],[186,-3.1429],[191,-3.1518],[193,-3.1458],[195,-3.1301],[196,-3.1173],[206,-2.6271],[208,-2.5013],[209,-2.4294],[210,-2.3439],[211,-2.2463],[212,-2.1352],[213,-2.0116],[214,-1.8751],[215,-1.7168],[216,-1.5343],[217,-1.3198],[218,-1.0772],[219,-0.8159],[220,-0.5407],[221,-0.2563],[222,0.0342],[223,0.3231],[224,0.6076],[225,0.8829],[226,1.1445],[227,1.3864],[228,1.6041],[229,1.7948],[230,1.951],[231,2.0818],[232,2.1919],[233,2.2857],[234,2.3616],[238,2.4723],[245,2.5258],[246,2.5123],[247,2.4899],[248,2.4596],[249,2.4212],[250,2.3765],[252,2.2698],[255,2.0807],[258,1.8794],[260,1.7583],[261,1.7061],[262,1.6579],[263,1.6173],[264,1.5811],[265,1.5552],[266,1.5366],[267,1.5274],[268,1.5277],[269,1.541],[271,1.5757],[274,1.6157],[278,1.6548],[290,1.7325],[292,1.7412],[294,1.7387],[296,1.7227],[297,1.7106],[298,1.6903],[299,1.6647],[301,1.5893],[302,1.542],[303,1.4895],[305,1.3719],[306,1.3084],[308,1.1714],[313,0.8154],[315,0.6819],[316,0.6173],[317,0.557],[318,0.4988],[319,0.4446],[320,0.3944],[354,0.2767]]
       }
       spine_03_ctrl.Scale.X {type=MovieSceneScriptingFloatChannel, default=1, const [1,315]}
       spine_03_ctrl.Scale.Y {type=MovieSceneScriptingFloatChannel, default=1, const [1,315]}
       spine_03_ctrl.Scale.Z {type=MovieSceneScriptingFloatChannel, default=1, const [1,315]}
       spine_04_ctrl.Location.X {
        type=MovieSceneScriptingFloatChannel
        default=0
        thinned_from=315
        keys [[0,-0.0401],[8,-0.0413],[11,-0.0413],[12,-0.0424],[15,-0.0424],[16,-0.0436],[17,-0.0437],[18,-0.0442],[19,-0.0444],[20,-0.0449],[22,-0.0454],[23,-0.0459],[25,-0.0464],[26,-0.0464],[28,-0.0468],[29,-0.0468],[30,-0.0471],[40,-0.047],[41,-0.0484],[86,-0.0461],[90,-0.0464],[95,-0.0474],[100,-0.0476],[101,-0.0454],[103,-0.0454],[104,-0.0431],[105,-0.0431],[106,-0.0408],[107,-0.0408],[109,-0.0362],[110,-0.0362],[112,-0.0317],[113,-0.0303],[114,-0.0275],[115,-0.0275],[116,-0.0247],[117,-0.0247],[119,-0.0192],[120,-0.0192],[121,-0.0164],[123,-0.0164],[124,-0.0136],[126,-0.0136],[127,-0.0109],[128,-0.0109],[129,-0.0106],[130,-0.0099],[131,-0.0099],[132,-0.0092],[135,-0.0092],[136,-0.0085],[141,-0.0085],[142,-0.0092],[147,-0.0093],[148,-0.0103],[152,-0.0103],[153,-0.0114],[158,-0.0114],[159,-0.0124],[176,-0.0123],[177,-0.0132],[179,-0.0132],[180,-0.0142],[183,-0.0142],[184,-0.0151],[190,-0.0151],[191,-0.016],[192,-0.016],[193,-0.0163],[204,-0.0163],[205,-0.0151],[209,-0.0151],[210,-0.0142],[212,-0.0142],[213,-0.0132],[215,-0.0132],[216,-0.0123],[218,-0.0123],[219,-0.0132],[221,-0.0132],[222,-0.0142],[224,-0.0142],[226,-0.0143],[227,-0.0149],[231,-0.0149],[232,-0.0155],[233,-0.0155],[234,-0.0161],[235,-0.0161],[237,-0.0172],[238,-0.0172],[240,-0.0184],[241,-0.0184],[247,-0.0201],[249,-0.0201],[251,-0.0207],[252,-0.0207],[253,-0.021],[254,-0.0216],[255,-0.0219],[260,-0.0253],[261,-0.0253],[263,-0.0268],[264,-0.0268],[265,-0.0275],[268,-0.0275],[269,-0.0282],[272,-0.0282],[273,-0.028],[288,-0.028],[289,-0.0292],[290,-0.0292],[291,-0.03],[293,-0.03],[294,-0.0308],[296,-0.0308],[297,-0.0317],[298,-0.0317],[299,-0.0325],[300,-0.0325],[302,-0.0341],[303,-0.0341],[304,-0.035],[306,-0.0351],[307,-0.0369],[310,-0.0369],[311,-0.0387],[315,-0.0387],[316,-0.0404],[354,-0.0401]]
       }
       spine_04_ctrl.Location.Y {
        type=MovieSceneScriptingFloatChannel
        default=0
        thinned_from=315
        keys [[0,-0.1025],[8,-0.1052],[10,-0.1052],[11,-0.1082],[13,-0.1082],[14,-0.1112],[15,-0.1112],[16,-0.1115],[18,-0.1125],[21,-0.1136],[30,-0.1139],[31,-0.1136],[35,-0.1137],[36,-0.114],[40,-0.114],[41,-0.1118],[48,-0.1118],[49,-0.112],[50,-0.1115],[52,-0.1114],[53,-0.111],[55,-0.1108],[56,-0.1104],[58,-0.1102],[59,-0.1096],[60,-0.1096],[61,-0.109],[62,-0.1089],[63,-0.1084],[64,-0.1083],[74,-0.1051],[77,-0.1047],[78,-0.1045],[85,-0.104],[86,-0.1032],[87,-0.1031],[89,-0.1015],[90,-0.1005],[91,-0.0988],[92,-0.098],[96,-0.0912],[97,-0.0904],[98,-0.0898],[99,-0.087],[100,-0.0865],[101,-0.0836],[102,-0.0831],[103,-0.0803],[104,-0.0797],[105,-0.0764],[106,-0.0735],[107,-0.073],[108,-0.073],[109,-0.0702],[110,-0.0696],[111,-0.0668],[114,-0.0666],[115,-0.0643],[117,-0.0643],[118,-0.0648],[119,-0.0648],[120,-0.0675],[121,-0.0675],[122,-0.0703],[123,-0.0707],[124,-0.073],[126,-0.0785],[127,-0.0808],[129,-0.0861],[130,-0.0867],[131,-0.0936],[132,-0.0936],[133,-0.1011],[135,-0.1011],[136,-0.1086],[143,-0.1086],[144,-0.1081],[145,-0.1066],[146,-0.1066],[147,-0.1061],[148,-0.1026],[151,-0.1026],[152,-0.0985],[156,-0.0985],[157,-0.0944],[176,-0.0943],[177,-0.0976],[178,-0.0976],[179,-0.1006],[180,-0.1009],[182,-0.1009],[183,-0.1038],[184,-0.1038],[185,-0.1041],[187,-0.1041],[188,-0.1071],[190,-0.1071],[191,-0.1074],[198,-0.1074],[199,-0.1077],[203,-0.1077],[204,-0.1036],[213,-0.1034],[214,-0.0983],[220,-0.0982],[221,-0.0931],[223,-0.0932],[224,-0.088],[225,-0.0856],[226,-0.0818],[227,-0.0778],[228,-0.0778],[230,-0.0703],[231,-0.07],[232,-0.0662],[233,-0.0662],[234,-0.0622],[235,-0.0619],[236,-0.0619],[237,-0.0616],[238,-0.0578],[241,-0.0569],[244,-0.0539],[245,-0.0533],[246,-0.0524],[251,-0.0494],[256,-0.0482],[257,-0.0486],[258,-0.0484],[259,-0.0486],[260,-0.0484],[262,-0.0486],[264,-0.0482],[266,-0.0482],[267,-0.0476],[268,-0.0476],[272,-0.0461],[273,-0.0459],[288,-0.0459],[289,-0.0434],[293,-0.0432],[296,-0.0429],[301,-0.0431],[304,-0.0434],[306,-0.0432],[307,-0.0443],[310,-0.0443],[311,-0.044],[312,-0.044],[313,-0.0453],[316,-0.0453],[317,-0.045],[318,-0.0463],[320,-0.0463],[354,-0.1025]]
       }
       spine_04_ctrl.Location.Z {
        type=MovieSceneScriptingFloatChannel
        default=0
        thinned_from=315
        keys [[0,0.001],[10,0.0008],[11,0.0042],[13,0.0042],[14,0.0076],[15,0.0076],[16,0.0113],[17,0.0124],[20,0.0198],[21,0.0248],[30,0.0474],[31,0.0474],[32,0.0499],[33,0.0493],[35,0.0493],[36,0.0533],[40,0.0533],[48,0.0535],[49,0.0523],[50,0.0524],[51,0.0503],[53,0.0504],[54,0.0483],[56,0.0483],[57,0.0463],[58,0.0463],[59,0.0443],[60,0.0443],[62,0.0402],[63,0.0402],[64,0.0382],[68,0.0338],[69,0.0316],[74,0.0261],[75,0.0261],[78,0.0228],[79,0.0228],[80,0.0217],[81,0.0209],[82,0.0209],[83,0.0196],[84,0.0196],[85,0.0182],[86,0.0183],[87,0.017],[88,0.017],[89,0.0158],[90,0.0131],[93,0.0094],[94,0.0069],[95,0.0057],[96,0.0031],[97,0.002],[98,-0.0046],[99,-0.0043],[100,-0.0109],[101,-0.0107],[102,-0.0172],[103,-0.017],[104,-0.0236],[105,-0.0299],[106,-0.0297],[107,-0.0363],[108,-0.0363],[109,-0.036],[110,-0.0426],[111,-0.0424],[112,-0.0424],[113,-0.0429],[117,-0.0427],[118,-0.0374],[119,-0.0374],[120,-0.0323],[121,-0.0323],[123,-0.0218],[124,-0.022],[126,-0.0117],[127,-0.0118],[128,-0.0067],[129,-0.0046],[130,0.0015],[131,0.001],[132,0.001],[133,0.0065],[135,0.0065],[136,0.0121],[143,0.0121],[144,0.0059],[145,0.0088],[146,0.0088],[147,0.0028],[148,0.0031],[151,0.0031],[152,-0.0025],[156,-0.0025],[157,-0.0082],[160,-0.0082],[161,-0.0086],[176,-0.0086],[177,-0.0079],[178,-0.0079],[179,-0.0081],[180,-0.0041],[182,-0.0041],[183,-0.0043],[184,-0.0043],[185,-0.0003],[187,-0.0003],[188,-0.0005],[190,-0.0005],[191,0.0035],[192,0.0035],[193,0.0029],[198,0.0029],[199,0.0069],[203,0.0069],[204,0.0073],[208,0.0073],[209,0.0063],[212,0.0063],[213,0.0054],[214,0.0058],[216,0.0058],[217,0.0048],[220,0.0048],[221,0.0053],[222,0.0062],[223,0.0062],[224,0.0057],[225,0.0065],[226,0.0068],[227,0.0036],[228,0.0036],[230,0.0042],[231,0.0006],[232,0.001],[233,0.001],[234,-0.0023],[235,-0.0058],[236,-0.0058],[237,-0.0093],[238,-0.009],[240,-0.0161],[241,-0.0178],[256,-0.0582],[257,-0.0588],[261,-0.0681],[262,-0.0681],[264,-0.0727],[266,-0.0727],[267,-0.075],[272,-0.0748],[273,-0.0741],[288,-0.0741],[289,-0.0707],[294,-0.0707],[295,-0.072],[296,-0.072],[297,-0.0733],[298,-0.0733],[299,-0.0747],[300,-0.0747],[301,-0.076],[302,-0.076],[304,-0.0787],[305,-0.0781],[306,-0.0781],[307,-0.0816],[310,-0.0816],[311,-0.0851],[316,-0.0852],[317,-0.0886],[320,-0.0887],[354,0.001]]
       }
       spine_04_ctrl.Rotation.X {
        type=MovieSceneScriptingFloatChannel
        default=0
        thinned_from=315
        keys [[0,-4.6667],[8,-5.105],[10,-5.0962],[11,-5.1107],[12,-5.1499],[13,-5.2018],[15,-5.338],[23,-6.0381],[25,-6.1849],[34,-6.6226],[44,-6.9823],[51,-7.1321],[54,-7.1541],[59,-7.1017],[81,-6.483],[83,-6.4518],[84,-6.4256],[85,-6.3819],[87,-6.2595],[88,-6.2118],[90,-6.1524],[92,-6.1301],[97,-6.1319],[100,-6.0981],[102,-6.0533],[103,-6.0205],[104,-5.9735],[105,-5.9082],[106,-5.8217],[107,-5.7151],[108,-5.5894],[109,-5.4423],[114,-4.6299],[116,-4.2897],[120,-3.5466],[121,-3.3678],[122,-3.1964],[123,-3.0322],[124,-2.8776],[125,-2.7328],[126,-2.5954],[128,-2.3478],[129,-2.2316],[131,-2.012],[133,-1.8092],[134,-1.7121],[135,-1.6236],[136,-1.5395],[137,-1.4642],[138,-1.4058],[139,-1.3523],[140,-1.3116],[141,-1.2793],[142,-1.2512],[143,-1.2316],[144,-1.2204],[145,-1.2128],[146,-1.2117],[152,-1.2457],[154,-1.2511],[156,-1.2504],[161,-1.2256],[162,-1.2246],[163,-1.2274],[164,-1.2235],[165,-1.2048],[166,-1.1768],[167,-1.1382],[168,-1.0927],[169,-1.0419],[170,-0.9857],[172,-0.8635],[175,-0.668],[176,-0.607],[177,-0.5502],[178,-0.4945],[179,-0.4502],[180,-0.4068],[181,-0.379],[182,-0.3536],[183,-0.342],[184,-0.3373],[185,-0.3388],[186,-0.3454],[187,-0.3596],[188,-0.3796],[189,-0.4044],[190,-0.4341],[191,-0.4682],[192,-0.5063],[193,-0.5472],[194,-0.592],[195,-0.6397],[196,-0.6896],[197,-0.741],[204,-1.1124],[206,-1.2115],[208,-1.3024],[210,-1.3844],[212,-1.4558],[214,-1.5141],[215,-1.5371],[216,-1.5478],[217,-1.5431],[220,-1.5197],[221,-1.5081],[222,-1.4913],[223,-1.4702],[225,-1.4149],[226,-1.3909],[227,-1.3727],[228,-1.3629],[229,-1.3647],[230,-1.3802],[231,-1.4132],[232,-1.4675],[233,-1.5465],[234,-1.6453],[235,-1.7532],[237,-1.9759],[238,-2.0818],[240,-2.2831],[241,-2.3971],[242,-2.5246],[244,-2.8119],[245,-2.9483],[246,-3.0771],[247,-3.1981],[249,-3.4189],[254,-3.918],[257,-4.1984],[258,-4.2987],[260,-4.5251],[264,-5.0423],[266,-5.2789],[268,-5.4823],[274,-5.9645],[282,-6.5055],[285,-6.6665],[289,-6.815],[292,-6.8736],[295,-6.8889],[313,-6.7524],[320,-6.8174],[354,-4.6667]]
       }
       spine_04_ctrl.Rotation.Y {
        type=MovieSceneScriptingFloatChannel
        default=0
        thinned_from=315
        keys [[0,0.0067],[8,0.0555],[9,0.0606],[10,0.0578],[11,0.0412],[12,0.0076],[13,-0.0406],[14,-0.1029],[15,-0.1775],[16,-0.2648],[17,-0.3645],[18,-0.4729],[19,-0.5916],[20,-0.717],[21,-0.8467],[23,-1.1095],[24,-1.2353],[25,-1.3548],[26,-1.4619],[27,-1.5582],[28,-1.6492],[29,-1.7295],[30,-1.7984],[31,-1.859],[32,-1.9118],[33,-1.9576],[35,-2.0292],[37,-2.089],[40,-2.1594],[42,-2.1927],[45,-2.2273],[49,-2.2422],[52,-2.2235],[55,-2.1542],[57,-2.0879],[58,-2.0504],[60,-1.9628],[62,-1.8669],[63,-1.8128],[64,-1.7625],[70,-1.4395],[72,-1.339],[74,-1.2469],[75,-1.2053],[76,-1.1662],[77,-1.13],[78,-1.0963],[79,-1.0649],[80,-1.0359],[81,-1.0098],[84,-0.9416],[85,-0.9143],[86,-0.893],[88,-0.8639],[89,-0.8467],[90,-0.8359],[91,-0.8357],[93,-0.8472],[94,-0.8508],[95,-0.8479],[96,-0.8338],[97,-0.8034],[98,-0.7473],[99,-0.6568],[100,-0.5401],[101,-0.4035],[102,-0.2482],[103,-0.0743],[104,0.1185],[105,0.3277],[106,0.5521],[107,0.7865],[108,1.0241],[110,1.494],[111,1.7246],[112,1.9493],[113,2.1647],[114,2.3706],[115,2.5665],[116,2.744],[117,2.8986],[118,3.0244],[119,3.1255],[120,3.202],[121,3.2542],[122,3.2798],[123,3.2854],[124,3.2715],[125,3.2389],[126,3.1883],[127,3.1219],[128,3.0444],[129,2.9575],[131,2.7734],[132,2.6872],[133,2.6088],[134,2.5398],[135,2.4831],[136,2.4381],[137,2.4073],[138,2.3904],[139,2.389],[140,2.3995],[142,2.4347],[144,2.4861],[149,2.6637],[152,2.7786],[155,2.8775],[159,2.9689],[161,2.9958],[168,3.011],[170,3.0057],[172,2.9892],[173,2.9752],[175,2.9292],[176,2.899],[178,2.8186],[180,2.716],[187,2.3376],[189,2.2389],[191,2.149],[193,2.0707],[195,2.008],[197,1.9586],[199,1.9205],[201,1.8961],[203,1.8856],[205,1.8893],[206,1.8966],[207,1.9078],[208,1.9229],[209,1.9426],[210,1.9685],[211,2],[212,2.0375],[213,2.081],[214,2.1332],[215,2.1917],[216,2.2373],[217,2.2579],[218,2.2522],[219,2.2215],[220,2.1721],[221,2.1115],[222,2.0464],[223,1.9945],[224,1.9513],[226,1.8796],[227,1.8493],[228,1.8238],[229,1.804],[230,1.7923],[231,1.7899],[232,1.7996],[233,1.8234],[234,1.8678],[235,1.9328],[236,2.0131],[237,2.112],[238,2.2267],[239,2.3567],[240,2.5006],[241,2.6527],[243,2.9755],[244,3.1469],[245,3.3269],[247,3.7064],[249,4.0949],[250,4.2743],[251,4.439],[252,4.5875],[253,4.7166],[254,4.8263],[255,4.9147],[256,4.9808],[258,5.076],[261,5.1769],[270,5.3854],[273,5.4183],[275,5.4138],[295,5.1464],[299,5.0574],[303,4.9188],[312,4.4989],[314,4.4246],[316,4.3702],[320,4.3195],[354,0.0067]]
       }
       spine_04_ctrl.Rotation.Z {
        type=MovieSceneScriptingFloatChannel
        default=0
        thinned_from=315
        keys [[0,0.9059],[8,0.4282],[9,0.3329],[10,0.2345],[11,0.1338],[12,0.031],[13,-0.0767],[15,-0.2963],[18,-0.6359],[19,-0.7505],[21,-0.9739],[22,-1.0816],[23,-1.1864],[24,-1.2857],[25,-1.3797],[26,-1.4648],[27,-1.5422],[28,-1.6112],[29,-1.6724],[30,-1.7215],[31,-1.7611],[32,-1.7942],[33,-1.8184],[34,-1.838],[38,-1.8993],[42,-1.9431],[46,-1.9757],[51,-1.9955],[56,-1.9858],[61,-1.9549],[73,-1.8637],[81,-1.8371],[83,-1.8376],[84,-1.8509],[85,-1.8848],[86,-1.9477],[87,-2.0395],[88,-2.161],[89,-2.3066],[90,-2.4811],[91,-2.6839],[92,-2.9059],[95,-3.6145],[96,-3.8371],[97,-4.0388],[98,-4.2098],[99,-4.349],[100,-4.4583],[101,-4.5429],[102,-4.6022],[103,-4.6377],[104,-4.6483],[105,-4.633],[106,-4.592],[107,-4.5319],[109,-4.3841],[113,-4.0483],[115,-3.8951],[116,-3.8106],[119,-3.5264],[120,-3.4435],[121,-3.376],[122,-3.33],[123,-3.2992],[124,-3.2792],[125,-3.2721],[126,-3.2832],[127,-3.3095],[128,-3.3474],[132,-3.5467],[134,-3.6325],[137,-3.7288],[138,-3.7528],[140,-3.8127],[146,-3.9228],[158,-4.0473],[168,-4.054],[174,-4.0251],[176,-4.0297],[178,-4.0596],[182,-4.1646],[184,-4.1943],[185,-4.1995],[194,-4.0788],[199,-3.9689],[202,-3.8767],[205,-3.7593],[208,-3.6119],[211,-3.4331],[213,-3.2955],[214,-3.2176],[215,-3.1259],[216,-3.0125],[217,-2.876],[218,-2.7179],[219,-2.5359],[220,-2.3272],[221,-2.0827],[222,-1.8043],[223,-1.4974],[224,-1.1734],[225,-0.8403],[227,-0.1613],[228,0.1751],[229,0.5013],[230,0.8178],[231,1.1175],[232,1.3992],[233,1.6591],[234,1.8863],[235,2.0764],[236,2.23],[237,2.3598],[238,2.4645],[239,2.5457],[240,2.6043],[241,2.6407],[242,2.6562],[243,2.652],[244,2.6293],[245,2.5877],[246,2.529],[247,2.4533],[248,2.3612],[249,2.2577],[250,2.1467],[254,1.6775],[255,1.5654],[256,1.4614],[257,1.3642],[258,1.2721],[259,1.1855],[260,1.106],[261,1.0333],[262,0.9682],[263,0.9096],[264,0.8586],[265,0.8158],[266,0.7836],[267,0.7626],[268,0.7544],[269,0.7593],[273,0.8119],[278,0.8911],[284,0.974],[285,0.9812],[286,1.0058],[287,1.0139],[288,1.0403],[289,1.0493],[290,1.0759],[291,1.0851],[292,1.1116],[293,1.1272],[294,1.1394],[295,1.1478],[296,1.1529],[297,1.1553],[298,1.1509],[299,1.1388],[300,1.1187],[301,1.0921],[302,1.0602],[303,1.0223],[304,0.9807],[305,0.9354],[306,0.887],[307,0.835],[308,0.781],[309,0.7247],[310,0.666],[311,0.6049],[313,0.4771],[314,0.4123],[315,0.3446],[316,0.2761],[317,0.2062],[318,0.1357],[319,0.0711],[320,0.0106],[354,0.9059]]
       }
       spine_04_ctrl.Scale.X {type=MovieSceneScriptingFloatChannel, default=1, const [1,315]}
       spine_04_ctrl.Scale.Y {type=MovieSceneScriptingFloatChannel, default=1, const [1,315]}
       spine_04_ctrl.Scale.Z {type=MovieSceneScriptingFloatChannel, default=1, const [1,315]}
       head_ctrl.Location.X {
        type=MovieSceneScriptingFloatChannel
        default=0
        thinned_from=322
        keys [[0,0.2506],[1,0.2513],[2,0.251],[3,0.2489],[6,0.2476],[7,0.2476],[8,0.247],[9,0.2445],[13,0.2415],[14,0.2393],[21,0.2328],[27,0.2282],[28,0.2313],[33,0.2299],[34,0.2306],[35,0.2305],[36,0.2313],[37,0.2312],[38,0.232],[39,0.2319],[40,0.2326],[41,0.2325],[42,0.2333],[43,0.2332],[44,0.234],[46,0.2338],[47,0.2346],[48,0.2346],[49,0.235],[60,0.2345],[61,0.231],[65,0.2309],[66,0.2304],[67,0.2304],[70,0.2287],[71,0.2287],[72,0.2282],[73,0.2282],[74,0.2276],[78,0.2276],[82,0.2303],[85,0.2328],[87,0.2331],[89,0.2316],[90,0.23],[92,0.2252],[95,0.2156],[97,0.2087],[98,0.2057],[99,0.2018],[100,0.1996],[101,0.1965],[102,0.195],[103,0.1927],[104,0.192],[105,0.1892],[106,0.1907],[107,0.1887],[108,0.1906],[109,0.19],[110,0.1912],[111,0.1932],[112,0.1927],[113,0.1941],[114,0.1942],[115,0.1961],[116,0.1961],[117,0.1979],[118,0.1979],[119,0.1991],[120,0.1988],[122,0.1997],[123,0.1983],[124,0.1978],[125,0.196],[126,0.1953],[127,0.1939],[128,0.1913],[133,0.1854],[135,0.1836],[136,0.1788],[137,0.1773],[144,0.1756],[152,0.1747],[153,0.1748],[154,0.1745],[157,0.1748],[158,0.1745],[160,0.1749],[161,0.1746],[162,0.1762],[165,0.1763],[166,0.1781],[167,0.177],[170,0.177],[171,0.1758],[172,0.1742],[173,0.1741],[174,0.1714],[175,0.1711],[176,0.1692],[177,0.1682],[179,0.1669],[180,0.1601],[181,0.1585],[183,0.1571],[185,0.1563],[192,0.1562],[193,0.1569],[199,0.1581],[200,0.1609],[201,0.1598],[207,0.1633],[212,0.1675],[213,0.1622],[215,0.1647],[216,0.1685],[217,0.1638],[218,0.1662],[219,0.1689],[220,0.1676],[221,0.1715],[222,0.1759],[223,0.1824],[224,0.1873],[225,0.1936],[226,0.2012],[227,0.2061],[228,0.2128],[229,0.2191],[230,0.2247],[231,0.2323],[233,0.2396],[234,0.2453],[238,0.2542],[239,0.2546],[240,0.2564],[241,0.2568],[242,0.2597],[243,0.2588],[245,0.2583],[246,0.2609],[247,0.2598],[250,0.2583],[251,0.2607],[252,0.2595],[266,0.2507],[269,0.2502],[279,0.2509],[280,0.2515],[285,0.2515],[286,0.252],[287,0.2518],[290,0.2528],[291,0.2546],[293,0.2548],[294,0.2565],[297,0.2564],[298,0.2581],[306,0.2551],[308,0.2538],[310,0.2532],[311,0.2524],[312,0.2504],[315,0.2491],[316,0.2468],[318,0.2458],[319,0.2441],[320,0.2437],[354,0.2506]]
       }
       head_ctrl.Location.Y {
        type=MovieSceneScriptingFloatChannel
        default=0
        thinned_from=322
        keys [[0,-0.1298],[1,-0.1354],[2,-0.1359],[3,-0.1372],[6,-0.1397],[7,-0.1453],[9,-0.1478],[10,-0.1486],[13,-0.1524],[14,-0.158],[21,-0.1669],[22,-0.1758],[27,-0.1809],[29,-0.1823],[30,-0.1836],[38,-0.1873],[39,-0.1874],[40,-0.1883],[42,-0.1889],[44,-0.1899],[45,-0.19],[47,-0.1909],[49,-0.191],[50,-0.1923],[60,-0.1929],[61,-0.1923],[63,-0.1923],[64,-0.1912],[65,-0.192],[69,-0.1918],[70,-0.19],[72,-0.1901],[74,-0.1898],[75,-0.188],[76,-0.188],[79,-0.183],[80,-0.1798],[81,-0.1779],[84,-0.1688],[85,-0.166],[86,-0.1637],[87,-0.1621],[88,-0.1614],[89,-0.1614],[90,-0.1628],[91,-0.1649],[92,-0.1681],[93,-0.1717],[94,-0.1762],[97,-0.1915],[98,-0.2004],[99,-0.2012],[100,-0.2091],[101,-0.2094],[102,-0.2166],[103,-0.2162],[104,-0.2233],[105,-0.2294],[106,-0.2285],[107,-0.2343],[108,-0.2333],[109,-0.2318],[110,-0.2379],[112,-0.2351],[114,-0.2347],[115,-0.2334],[116,-0.2334],[117,-0.2321],[119,-0.2311],[120,-0.2301],[121,-0.2302],[122,-0.2309],[123,-0.2308],[125,-0.2329],[126,-0.2335],[127,-0.2358],[128,-0.2375],[129,-0.2379],[135,-0.2428],[136,-0.2429],[137,-0.2559],[144,-0.2569],[146,-0.2593],[147,-0.2592],[148,-0.2604],[149,-0.2602],[150,-0.2613],[151,-0.2612],[152,-0.2623],[153,-0.2622],[154,-0.2632],[157,-0.2629],[158,-0.264],[160,-0.2639],[161,-0.2632],[165,-0.2632],[166,-0.2609],[167,-0.2625],[168,-0.2625],[169,-0.26],[170,-0.26],[171,-0.2617],[173,-0.2616],[174,-0.2605],[179,-0.2617],[180,-0.2613],[181,-0.2628],[182,-0.2631],[183,-0.2653],[192,-0.2653],[193,-0.2645],[194,-0.2644],[195,-0.2655],[200,-0.2648],[201,-0.266],[208,-0.2632],[211,-0.2611],[212,-0.2645],[214,-0.2625],[215,-0.2657],[216,-0.2624],[217,-0.2648],[219,-0.2617],[220,-0.257],[221,-0.2546],[222,-0.2475],[223,-0.2412],[224,-0.2331],[225,-0.2265],[226,-0.2173],[230,-0.1779],[231,-0.1673],[232,-0.1547],[233,-0.1456],[234,-0.137],[235,-0.129],[236,-0.1221],[237,-0.1158],[238,-0.1105],[239,-0.1095],[240,-0.105],[242,-0.1021],[243,-0.0993],[244,-0.0992],[245,-0.0994],[246,-0.0989],[247,-0.0971],[249,-0.0989],[250,-0.1],[251,-0.1004],[252,-0.0996],[256,-0.1058],[257,-0.107],[258,-0.1085],[259,-0.1103],[260,-0.1117],[261,-0.1125],[262,-0.1147],[263,-0.1152],[264,-0.1172],[265,-0.1175],[266,-0.1191],[267,-0.1197],[268,-0.121],[269,-0.1205],[270,-0.1214],[271,-0.1214],[272,-0.1224],[273,-0.1229],[275,-0.1227],[276,-0.1255],[279,-0.1251],[280,-0.1286],[283,-0.1283],[284,-0.1291],[285,-0.129],[286,-0.1316],[289,-0.1285],[290,-0.1291],[291,-0.1234],[292,-0.124],[293,-0.124],[294,-0.1184],[297,-0.1187],[298,-0.1133],[299,-0.1129],[300,-0.1134],[301,-0.1132],[303,-0.1147],[304,-0.1148],[305,-0.1153],[306,-0.1162],[307,-0.1155],[308,-0.1165],[309,-0.1205],[310,-0.1216],[311,-0.1209],[313,-0.1232],[314,-0.1272],[315,-0.1283],[316,-0.1277],[318,-0.1296],[319,-0.1336],[320,-0.1344],[354,-0.1298]]
       }
       head_ctrl.Location.Z {
        type=MovieSceneScriptingFloatChannel
        default=0
        thinned_from=322
        keys [[0,-0.006],[1,-0.0062],[2,-0.0055],[3,-0.0063],[6,-0.0046],[7,-0.0028],[9,-0.002],[10,-0.0002],[13,0.0013],[14,0.0048],[18,0.0073],[21,0.0084],[22,0.0105],[24,0.0106],[29,0.0092],[30,0.005],[32,0.0043],[33,0.0037],[34,0.0019],[35,0.0017],[36,-0.0001],[37,-0.0003],[38,-0.0021],[39,-0.0024],[40,-0.0041],[41,-0.0043],[42,-0.0061],[43,-0.0062],[44,-0.0079],[46,-0.0082],[47,-0.0098],[48,-0.0099],[49,-0.0114],[50,-0.0112],[56,-0.0106],[60,-0.0096],[61,-0.0063],[63,-0.0057],[64,-0.0056],[66,-0.0047],[67,-0.0029],[69,-0.0022],[70,-0.0009],[72,-0.0003],[73,0.0013],[74,0.0016],[75,0.0028],[76,0.003],[77,0.0042],[78,0.004],[79,0.0051],[80,0.0058],[81,0.006],[83,0.0075],[84,0.0085],[85,0.0097],[86,0.0113],[87,0.0132],[89,0.018],[91,0.0235],[95,0.0358],[96,0.0387],[97,0.0413],[98,0.0443],[99,0.045],[100,0.0478],[101,0.0472],[102,0.0498],[103,0.0491],[104,0.0494],[105,0.0507],[106,0.0497],[107,0.0498],[109,0.0454],[110,0.0454],[112,0.041],[113,0.0393],[114,0.0363],[115,0.035],[116,0.0318],[117,0.0303],[118,0.0265],[119,0.0225],[120,0.0209],[121,0.0168],[123,0.0134],[124,0.0092],[125,0.0078],[126,0.006],[127,0.0021],[128,0.0008],[129,-0.0019],[131,-0.0049],[133,-0.0074],[135,-0.0093],[136,-0.0094],[137,-0.007],[138,-0.0055],[140,-0.0056],[144,-0.0045],[145,-0.0041],[146,-0.0031],[147,-0.0026],[155,0.0044],[156,0.0049],[158,0.0065],[160,0.0072],[161,0.0081],[165,0.0087],[166,0.0079],[167,0.0013],[168,0.001],[169,0],[170,-0.0005],[171,-0.0077],[172,-0.0082],[173,-0.009],[174,-0.0168],[176,-0.0184],[177,-0.0198],[179,-0.0217],[180,-0.0216],[181,-0.0278],[182,-0.0285],[183,-0.0288],[192,-0.033],[193,-0.0336],[194,-0.034],[195,-0.034],[199,-0.0351],[200,-0.0357],[201,-0.0445],[208,-0.0456],[209,-0.0462],[211,-0.0462],[212,-0.0454],[213,-0.0443],[214,-0.0442],[215,-0.0432],[216,-0.0307],[217,-0.0288],[218,-0.0284],[219,-0.0276],[220,-0.0128],[221,-0.0108],[222,-0.0091],[223,0.0063],[224,0.0089],[225,0.0157],[226,0.0218],[227,0.0298],[229,0.0423],[230,0.047],[231,0.0529],[232,0.0563],[233,0.0607],[234,0.0632],[235,0.067],[236,0.0689],[237,0.0702],[238,0.0694],[239,0.0694],[240,0.0679],[241,0.0678],[242,0.0689],[243,0.0597],[246,0.0626],[247,0.0531],[250,0.0552],[251,0.0557],[252,0.0457],[257,0.0469],[258,0.047],[259,0.0448],[260,0.0449],[261,0.0446],[262,0.0426],[263,0.0422],[264,0.0401],[265,0.0395],[266,0.0372],[267,0.0367],[268,0.0342],[269,0.0333],[270,0.0308],[271,0.0303],[272,0.0278],[273,0.0268],[275,0.0263],[276,0.0202],[279,0.0199],[280,0.0143],[283,0.0143],[284,0.0148],[285,0.0148],[286,0.0089],[289,0.0079],[290,0.0064],[291,0.004],[292,0.0026],[293,0.0028],[294,0.0006],[297,0.0015],[298,-0.0006],[299,0.0015],[300,0.0019],[301,0.0041],[303,0.005],[304,0.0072],[305,0.0085],[306,0.009],[307,0.0131],[308,0.0136],[309,0.0155],[310,0.016],[311,0.0201],[313,0.021],[314,0.0229],[315,0.0233],[316,0.0273],[318,0.028],[319,0.0298],[320,0.0301],[354,-0.006]]
       }
       head_ctrl.Rotation.X {
        type=MovieSceneScriptingFloatChannel
        default=0
        thinned_from=322
        keys [[0,-5.7506],[1,-5.3693],[2,-5.2561],[3,-5.176],[4,-5.1261],[5,-5.1044],[6,-5.1087],[7,-5.1392],[8,-5.1867],[9,-5.2626],[10,-5.3614],[11,-5.4735],[12,-5.6053],[14,-5.9037],[18,-6.5295],[20,-6.796],[21,-6.9138],[23,-7.0911],[25,-7.2042],[27,-7.2423],[29,-7.2095],[30,-7.1671],[32,-7.0371],[33,-6.9529],[35,-6.7421],[37,-6.5027],[45,-5.4865],[47,-5.2714],[50,-5],[52,-4.8467],[53,-4.7907],[54,-4.7468],[55,-4.6932],[59,-4.5697],[63,-4.4867],[66,-4.4509],[68,-4.445],[70,-4.4731],[71,-4.5143],[72,-4.5662],[73,-4.6379],[74,-4.7273],[75,-4.8445],[76,-4.9927],[77,-5.1722],[78,-5.4146],[79,-5.734],[80,-6.1024],[81,-6.5142],[82,-6.9631],[86,-8.8483],[87,-9.2909],[88,-9.6959],[89,-10.0457],[91,-10.6073],[93,-11.1363],[94,-11.3839],[95,-11.6016],[96,-11.7905],[97,-11.9385],[98,-12.0133],[99,-12.0024],[100,-11.9288],[101,-11.795],[102,-11.6122],[103,-11.3861],[106,-10.5575],[108,-9.9856],[111,-9.2179],[113,-8.7376],[115,-8.3226],[116,-8.1505],[117,-8.0299],[118,-7.9572],[120,-7.9414],[121,-7.9886],[123,-8.1901],[125,-8.4954],[126,-8.6691],[129,-9.2717],[132,-9.8981],[134,-10.2781],[135,-10.4496],[136,-10.595],[137,-10.7074],[138,-10.7914],[140,-10.8528],[144,-10.8555],[157,-10.5182],[162,-10.2885],[163,-10.2673],[164,-10.2178],[165,-10.1048],[166,-9.9324],[168,-9.4927],[170,-8.9763],[171,-8.7342],[172,-8.5223],[173,-8.3549],[174,-8.2474],[175,-8.2024],[177,-8.2583],[180,-8.4593],[181,-8.5421],[184,-8.73],[187,-8.8668],[188,-8.8953],[189,-8.8513],[191,-8.8697],[192,-8.8629],[194,-8.8012],[196,-8.6885],[200,-8.3724],[204,-8.0075],[206,-7.851],[208,-7.7439],[209,-7.7102],[210,-7.6924],[211,-7.7072],[212,-7.7866],[213,-7.8931],[215,-8.1604],[216,-8.3321],[217,-8.5437],[218,-8.7959],[220,-9.3612],[222,-9.9675],[223,-10.2557],[224,-10.5224],[225,-10.7637],[226,-10.9745],[227,-11.1566],[228,-11.3075],[229,-11.4346],[230,-11.5334],[231,-11.6017],[232,-11.6351],[233,-11.6355],[234,-11.6012],[235,-11.5107],[236,-11.3848],[237,-11.215],[238,-11.0103],[239,-10.7527],[240,-10.4345],[241,-10.0553],[242,-9.6252],[243,-9.1451],[248,-6.4636],[249,-5.9437],[250,-5.4479],[251,-4.9754],[252,-4.5306],[253,-4.1182],[254,-3.7351],[255,-3.3834],[256,-3.0497],[257,-2.7342],[258,-2.4255],[260,-1.7893],[261,-1.4648],[263,-0.832],[264,-0.5206],[265,-0.2159],[266,0.0844],[267,0.3797],[268,0.6694],[269,0.9634],[271,1.5318],[273,2.0735],[274,2.3302],[275,2.5755],[276,2.809],[280,3.6968],[282,4.1183],[284,4.5017],[286,4.8405],[288,5.1551],[290,5.4508],[292,5.7196],[293,5.8251],[294,5.9098],[295,5.9668],[296,5.9944],[297,6.0051],[298,5.9897],[299,5.9418],[300,5.8699],[301,5.7766],[302,5.6661],[303,5.5381],[304,5.3923],[305,5.2354],[306,5.0673],[311,4.1445],[313,3.798],[314,3.6371],[315,3.4907],[316,3.3567],[317,3.2451],[318,3.1527],[319,3.0773],[320,3.0159],[354,-5.7506]]
       }
       head_ctrl.Rotation.Y {
        type=MovieSceneScriptingFloatChannel
        default=0
        thinned_from=322
        keys [[0,-4.3117],[1,-3.4371],[2,-3.0805],[3,-2.7387],[4,-2.4155],[5,-2.1057],[6,-1.8132],[7,-1.5327],[8,-1.2654],[9,-1.0114],[10,-0.7351],[11,-0.4007],[12,0.0085],[13,0.4682],[14,0.9722],[15,1.5186],[16,2.0984],[17,2.7007],[20,4.5803],[21,5.1965],[22,5.7862],[23,6.3423],[24,6.8525],[25,7.2913],[26,7.6628],[27,7.9697],[28,8.2364],[29,8.4664],[30,8.6489],[31,8.7942],[32,8.9023],[33,8.9823],[36,9.1164],[40,9.196],[47,9.1587],[51,9.0867],[52,9.0435],[53,9.0211],[56,8.8623],[59,8.6379],[68,7.7767],[73,7.2183],[75,6.9602],[77,6.665],[78,6.484],[79,6.265],[80,6.0198],[82,5.4756],[83,5.1901],[85,4.6003],[86,4.3351],[87,4.1482],[88,4.0503],[89,4.0319],[90,4.1458],[91,4.3656],[92,4.6501],[93,4.9799],[95,5.6639],[96,5.9786],[97,6.2358],[98,6.4087],[99,6.4781],[100,6.4535],[101,6.3643],[102,6.1971],[103,5.9501],[104,5.6313],[105,5.2144],[106,4.6984],[107,4.1038],[108,3.4618],[109,2.7769],[110,2.0674],[112,0.6307],[113,-0.0813],[114,-0.7708],[115,-1.4526],[116,-2.12],[117,-2.7617],[118,-3.3663],[119,-3.9127],[120,-4.402],[121,-4.817],[122,-5.1478],[123,-5.4115],[124,-5.6102],[125,-5.7475],[126,-5.8119],[127,-5.8239],[128,-5.7991],[129,-5.7535],[133,-5.4545],[134,-5.3875],[136,-5.2828],[139,-5.1634],[144,-4.995],[152,-4.715],[155,-4.6456],[157,-4.6271],[159,-4.6353],[162,-4.6954],[163,-4.7229],[164,-4.7848],[165,-4.9003],[166,-5.049],[167,-5.2282],[171,-6.0489],[172,-6.2447],[173,-6.4136],[174,-6.5397],[175,-6.6229],[176,-6.6775],[177,-6.7076],[178,-6.686],[179,-6.6401],[181,-6.5],[185,-6.1695],[186,-6.1048],[191,-5.6763],[193,-5.5239],[195,-5.392],[198,-5.2392],[200,-5.1798],[202,-5.1564],[208,-5.1896],[209,-5.1836],[210,-5.1536],[211,-5.0176],[212,-4.653],[213,-4.2348],[214,-3.7891],[215,-3.3346],[216,-2.8723],[217,-2.3949],[218,-1.9006],[219,-1.4337],[220,-1.0046],[221,-0.6509],[222,-0.3697],[223,-0.1862],[224,-0.106],[225,-0.1058],[226,-0.1722],[227,-0.3065],[228,-0.5024],[229,-0.7501],[230,-1.0432],[231,-1.3796],[232,-1.7414],[233,-2.1239],[234,-2.5003],[235,-2.8859],[236,-3.289],[237,-3.7196],[238,-4.1787],[239,-4.6524],[240,-5.1406],[242,-6.1383],[244,-7.1103],[245,-7.5836],[246,-8.0403],[247,-8.473],[248,-8.8732],[249,-9.2363],[250,-9.5343],[251,-9.7693],[252,-9.9176],[253,-9.9806],[254,-9.9792],[255,-9.9214],[256,-9.8208],[257,-9.6805],[258,-9.5156],[259,-9.3257],[263,-8.4573],[264,-8.279],[265,-8.1201],[266,-7.9959],[267,-7.9051],[268,-7.8443],[269,-7.8206],[273,-7.7895],[285,-7.4721],[286,-7.4993],[287,-7.613],[288,-7.8037],[289,-8.0413],[290,-8.3258],[292,-8.9448],[293,-9.2363],[294,-9.493],[295,-9.6999],[296,-9.8474],[297,-9.9742],[298,-10.0536],[299,-10.0907],[300,-10.0923],[301,-10.0696],[302,-10.0205],[303,-9.95],[305,-9.7443],[307,-9.4604],[315,-8.0636],[317,-7.7477],[320,-7.3434],[354,-4.3117]]
       }
       head_ctrl.Rotation.Z {
        type=MovieSceneScriptingFloatChannel
        default=0
        thinned_from=322
        keys [[0,-1.2578],[1,-0.6105],[2,-0.3839],[3,-0.1608],[4,0.0568],[5,0.2686],[6,0.471],[7,0.6648],[8,0.851],[10,1.1988],[12,1.5297],[13,1.6877],[16,2.1206],[17,2.2559],[18,2.3826],[19,2.4969],[20,2.5887],[21,2.6684],[22,2.7273],[23,2.7742],[24,2.7933],[25,2.7916],[26,2.7742],[27,2.7435],[28,2.7033],[29,2.6509],[31,2.5284],[34,2.3183],[37,2.1252],[40,1.9451],[42,1.8342],[44,1.7384],[46,1.654],[47,1.6183],[48,1.5869],[49,1.56],[51,1.5209],[53,1.5031],[55,1.5073],[56,1.5165],[57,1.5309],[59,1.5706],[61,1.6218],[65,1.7493],[68,1.86],[69,1.8707],[70,1.8611],[71,1.8227],[72,1.7484],[73,1.6283],[74,1.4608],[75,1.2294],[76,0.9184],[77,0.4944],[78,-0.1238],[79,-0.9889],[80,-2.0326],[81,-3.2403],[82,-4.5462],[83,-5.9166],[84,-7.3344],[85,-8.7993],[86,-10.2425],[87,-11.6077],[88,-12.8335],[89,-13.874],[90,-14.6115],[91,-15.1795],[92,-15.6492],[93,-16.023],[94,-16.3106],[95,-16.5001],[96,-16.596],[97,-16.5822],[98,-16.4447],[99,-16.1859],[100,-15.8311],[101,-15.3622],[102,-14.8019],[103,-14.1524],[104,-13.4096],[106,-11.8296],[108,-10.3101],[109,-9.5825],[110,-8.8932],[111,-8.2422],[112,-7.6351],[113,-7.0779],[114,-6.5848],[115,-6.1333],[116,-5.7094],[117,-5.3274],[118,-4.9939],[119,-4.7095],[120,-4.4838],[121,-4.3182],[122,-4.2231],[123,-4.1824],[124,-4.1759],[125,-4.2076],[126,-4.289],[127,-4.4169],[128,-4.5994],[129,-4.8173],[131,-5.2975],[132,-5.523],[133,-5.7198],[134,-5.8937],[135,-6.0403],[136,-6.1566],[137,-6.2442],[138,-6.3039],[139,-6.3479],[141,-6.3673],[142,-6.3533],[144,-6.2862],[146,-6.17],[152,-5.7242],[154,-5.6026],[155,-5.5597],[156,-5.5304],[157,-5.5176],[158,-5.523],[159,-5.5482],[160,-5.5933],[161,-5.6646],[162,-5.7626],[163,-5.8912],[164,-6.0813],[165,-6.3647],[166,-6.7271],[167,-7.1446],[168,-7.6054],[171,-9.0778],[172,-9.5394],[173,-9.9542],[174,-10.3073],[175,-10.5927],[176,-10.8257],[178,-11.2262],[180,-11.5207],[182,-11.7265],[183,-11.7999],[184,-11.8465],[186,-11.8568],[192,-11.5147],[204,-11.0205],[205,-10.9302],[206,-10.8169],[207,-10.6671],[208,-10.4861],[209,-10.2556],[210,-9.9384],[211,-9.3679],[212,-8.2434],[213,-6.8972],[214,-5.3531],[215,-3.6502],[216,-1.8416],[217,0.0395],[218,1.9729],[219,3.9248],[220,5.8923],[221,7.8788],[224,13.912],[225,15.8402],[226,17.6879],[227,19.4175],[228,21.0126],[229,22.4473],[230,23.7064],[231,24.7562],[232,25.5815],[233,26.1584],[234,26.49],[235,26.6245],[236,26.5713],[237,26.3494],[238,25.9765],[239,25.4658],[240,24.8454],[241,24.1245],[242,23.3143],[243,22.4357],[245,20.526],[249,16.5258],[250,15.5548],[251,14.6219],[252,13.7426],[253,12.9224],[255,11.4337],[256,10.7128],[257,10.0176],[258,9.3517],[259,8.728],[260,8.1372],[261,7.5868],[262,7.0779],[263,6.6065],[264,6.1684],[265,5.7667],[266,5.409],[267,5.0942],[268,4.8228],[269,4.6119],[271,4.2446],[273,3.9158],[274,3.7668],[275,3.6316],[276,3.5067],[278,3.2866],[279,3.1845],[281,3.0082],[283,2.862],[284,2.7966],[285,2.7216],[286,2.5307],[287,2.0843],[288,1.4565],[289,0.6644],[290,-0.2326],[291,-1.2079],[292,-2.1868],[293,-3.1393],[294,-4.008],[295,-4.7403],[296,-5.2928],[297,-5.8068],[298,-6.1955],[299,-6.4896],[300,-6.7209],[301,-6.8867],[302,-7.0154],[303,-7.0923],[304,-7.1346],[305,-7.1343],[306,-7.0976],[307,-7.0203],[308,-6.9074],[309,-6.7751],[310,-6.6231],[312,-6.2867],[313,-6.1076],[314,-5.9085],[316,-5.4757],[318,-5.011],[320,-4.5686],[354,-1.2578]]
       }
       head_ctrl.Scale.X {type=MovieSceneScriptingFloatChannel, default=1, const [1,322]}
       head_ctrl.Scale.Y {type=MovieSceneScriptingFloatChannel, default=1, const [1,322]}
       head_ctrl.Scale.Z {type=MovieSceneScriptingFloatChannel, default=1, const [1,322]}
       clavicle_l_ctrl.Location.X {type=MovieSceneScriptingFloatChannel, default=0, thinned_from=322, keys [[0,0],[354,0]]}
       clavicle_l_ctrl.Location.Y {type=MovieSceneScriptingFloatChannel, default=0, thinned_from=322, keys [[0,0],[354,0]]}
       clavicle_l_ctrl.Location.Z {type=MovieSceneScriptingFloatChannel, default=0, thinned_from=322, keys [[0,0],[354,0]]}
       clavicle_l_ctrl.Rotation.X {type=MovieSceneScriptingFloatChannel, default=0, thinned_from=322, keys [[0,5.9102],[354,5.9102]]}
       clavicle_l_ctrl.Rotation.Y {type=MovieSceneScriptingFloatChannel, default=0, thinned_from=322, keys [[0,0.2415],[354,0.2415]]}
       clavicle_l_ctrl.Rotation.Z {type=MovieSceneScriptingFloatChannel, default=0, thinned_from=322, keys [[0,11.4074],[354,11.4074]]}
       clavicle_l_ctrl.Scale.X {type=MovieSceneScriptingFloatChannel, default=1, const [1,322]}
       clavicle_l_ctrl.Scale.Y {type=MovieSceneScriptingFloatChannel, default=1, const [1,322]}
       clavicle_l_ctrl.Scale.Z {type=MovieSceneScriptingFloatChannel, default=1, const [1,322]}
       clavicle_r_ctrl.Location.X {type=MovieSceneScriptingFloatChannel, default=0, thinned_from=322, keys [[0,0],[354,0]]}
       clavicle_r_ctrl.Location.Y {type=MovieSceneScriptingFloatChannel, default=0, thinned_from=322, keys [[0,0],[354,0]]}
       clavicle_r_ctrl.Location.Z {type=MovieSceneScriptingFloatChannel, default=0, thinned_from=322, keys [[0,0],[354,0]]}
       clavicle_r_ctrl.Rotation.X {type=MovieSceneScriptingFloatChannel, default=0, thinned_from=322, keys [[0,2.4237],[354,2.4237]]}
       clavicle_r_ctrl.Rotation.Y {type=MovieSceneScriptingFloatChannel, default=0, thinned_from=322, keys [[0,0.0791],[354,0.0791]]}
       clavicle_r_ctrl.Rotation.Z {type=MovieSceneScriptingFloatChannel, default=0, thinned_from=322, keys [[0,4.8999],[354,4.8999]]}
       clavicle_r_ctrl.Scale.X {type=MovieSceneScriptingFloatChannel, default=1, const [1,322]}
       clavicle_r_ctrl.Scale.Y {type=MovieSceneScriptingFloatChannel, default=1, const [1,322]}
       clavicle_r_ctrl.Scale.Z {type=MovieSceneScriptingFloatChannel, default=1, const [1,322]}
       neck_01_ctrl.Location.X {type=MovieSceneScriptingFloatChannel, default=0, thinned_from=322, keys [[0,0],[354,0]]}
       neck_01_ctrl.Location.Y {type=MovieSceneScriptingFloatChannel, default=0, thinned_from=322, keys [[0,0],[354,0]]}
       neck_01_ctrl.Location.Z {type=MovieSceneScriptingFloatChannel, default=0, thinned_from=322, keys [[0,0],[354,0]]}
       neck_01_ctrl.Rotation.X {
        type=MovieSceneScriptingFloatChannel
        default=0
        thinned_from=322
        keys [[0,-7.5136],[92,-7.5124],[94,-7.5538],[98,-7.7127],[107,-8.183],[110,-8.2849],[113,-8.3318],[123,-8.3607],[126,-8.3729],[127,-8.3906],[128,-8.3906],[129,-8.4095],[130,-8.4095],[131,-8.4285],[132,-8.4285],[145,-8.6066],[146,-8.6066],[147,-8.6318],[148,-8.6318],[149,-8.6508],[150,-8.6508],[151,-8.6697],[155,-8.6855],[218,-8.7012],[220,-8.6474],[223,-8.4928],[233,-7.7219],[235,-7.6032],[237,-7.5238],[239,-7.4954],[354,-7.5136]]
       }
       neck_01_ctrl.Rotation.Y {
        type=MovieSceneScriptingFloatChannel
        default=0
        thinned_from=248
        keys [[0,0.0808],[80,0.0808],[81,0.0813],[94,0.0813],[95,0.0791],[96,0.0813],[97,0.0808],[98,0.0808],[99,0.0814],[100,0.0808],[101,0.0806],[102,0.0811],[103,0.0806],[104,0.0811],[105,0.0803],[106,0.08],[107,0.0811],[108,0.0806],[109,0.0803],[110,0.0811],[111,0.0808],[128,0.0807],[129,0.0813],[130,0.0813],[131,0.0791],[133,0.0791],[134,0.0813],[135,0.0791],[136,0.0813],[137,0.0813],[138,0.0791],[144,0.0791],[145,0.0814],[148,0.0813],[149,0.0791],[150,0.0791],[151,0.0814],[152,0.0814],[153,0.0792],[160,0.0791],[161,0.078],[192,0.078],[199,0.0782],[200,0.1389],[201,0.3249],[202,0.6057],[203,0.9813],[240,8.6769],[241,8.5927],[242,8.4778],[243,8.3361],[244,8.1639],[245,7.9725],[246,7.7544],[247,7.5172],[248,7.2609],[250,6.6986],[252,6.0831],[255,5.0972],[259,3.7314],[260,3.3952],[262,2.7432],[264,2.1236],[265,1.8361],[266,1.5649],[267,1.3099],[268,1.0711],[269,0.8565],[270,0.6622],[271,0.4922],[272,0.3505],[273,0.2333],[274,0.1483],[275,0.1002],[276,0.081],[354,0.0808]]
       }
       neck_01_ctrl.Rotation.Z {
        type=MovieSceneScriptingFloatChannel
        default=0
        thinned_from=66
        keys [[0,0.0076],[72,0.0076],[86,-13.1987],[97,-18.1677],[134,-7.1942],[182,-3.4595],[196,-3.4595],[202,-1.9802],[211,8.0152],[217,14.8546],[224,17.8977],[265,0.5354],[290,-6.4184],[319,0.0073],[354,0.0076]]
       }
       neck_01_ctrl.Scale.X {type=MovieSceneScriptingFloatChannel, default=1, const [1,322]}
       neck_01_ctrl.Scale.Y {type=MovieSceneScriptingFloatChannel, default=1, const [1,322]}
       neck_01_ctrl.Scale.Z {type=MovieSceneScriptingFloatChannel, default=1, const [1,322]}
       neck_02_ctrl.Location.X {
        type=MovieSceneScriptingFloatChannel
        default=0
        thinned_from=322
        keys [[0,0.471],[2,0.4758],[5,0.4758],[6,0.4782],[19,0.478],[20,0.481],[25,0.481],[26,0.484],[30,0.484],[31,0.487],[34,0.4873],[35,0.4909],[39,0.4909],[40,0.4946],[44,0.4946],[45,0.4983],[64,0.4985],[65,0.4957],[68,0.4957],[69,0.4918],[74,0.4918],[75,0.4879],[78,0.4879],[79,0.4839],[81,0.4831],[86,0.474],[87,0.4703],[94,0.4574],[95,0.4574],[96,0.4556],[103,0.4545],[104,0.4555],[105,0.4555],[107,0.4574],[108,0.4574],[110,0.4593],[111,0.4593],[113,0.4605],[114,0.4619],[115,0.4619],[116,0.4634],[117,0.4634],[118,0.4648],[120,0.4648],[121,0.4662],[122,0.4662],[123,0.4677],[124,0.4677],[125,0.4691],[127,0.4691],[128,0.4705],[138,0.4719],[147,0.4723],[153,0.4716],[158,0.4708],[166,0.4695],[167,0.4733],[170,0.4733],[171,0.477],[173,0.477],[174,0.4808],[183,0.4808],[184,0.4862],[195,0.4862],[196,0.4891],[200,0.4891],[201,0.4919],[204,0.4919],[205,0.4947],[215,0.4952],[216,0.4916],[218,0.4916],[225,0.4662],[228,0.4522],[229,0.4522],[231,0.4428],[232,0.4428],[233,0.4382],[235,0.4382],[236,0.4335],[240,0.4335],[242,0.437],[243,0.437],[244,0.439],[245,0.439],[247,0.443],[249,0.443],[250,0.4449],[251,0.4449],[252,0.4469],[254,0.4469],[255,0.4489],[258,0.4489],[259,0.4513],[260,0.4513],[261,0.4537],[262,0.4537],[263,0.4561],[264,0.4561],[265,0.4585],[266,0.4585],[267,0.4609],[268,0.4609],[269,0.4633],[270,0.4633],[271,0.4656],[274,0.4664],[275,0.4714],[278,0.4714],[279,0.4764],[283,0.4764],[284,0.4813],[290,0.4813],[291,0.4798],[292,0.4798],[293,0.4782],[294,0.4782],[295,0.4766],[296,0.4766],[297,0.475],[298,0.475],[299,0.4734],[300,0.4734],[301,0.4719],[303,0.4719],[304,0.4703],[306,0.4703],[307,0.4686],[309,0.4686],[310,0.4668],[313,0.4668],[314,0.4651],[320,0.4651],[354,0.471]]
       }
       neck_02_ctrl.Location.Y {
        type=MovieSceneScriptingFloatChannel
        default=0
        thinned_from=322
        keys [[0,0.4243],[1,0.4238],[2,0.4295],[8,0.429],[9,0.4347],[14,0.4342],[15,0.4399],[19,0.4404],[20,0.4455],[23,0.4455],[24,0.4507],[28,0.4507],[29,0.4558],[32,0.4558],[33,0.4547],[34,0.4556],[38,0.4553],[64,0.4559],[65,0.4547],[66,0.4547],[67,0.4515],[75,0.4525],[76,0.4494],[78,0.4494],[79,0.4467],[80,0.4467],[84,0.442],[87,0.4394],[88,0.4396],[89,0.4389],[91,0.4406],[93,0.4406],[95,0.4382],[97,0.4342],[98,0.4343],[99,0.4249],[100,0.425],[101,0.4157],[102,0.4157],[103,0.4064],[104,0.4064],[106,0.3876],[107,0.3878],[108,0.3783],[110,0.3784],[111,0.369],[113,0.369],[114,0.3646],[115,0.3622],[116,0.3622],[118,0.3578],[120,0.3576],[121,0.3552],[123,0.3551],[124,0.3529],[125,0.3549],[126,0.3525],[127,0.3547],[128,0.3523],[129,0.3539],[131,0.3538],[132,0.3491],[138,0.3489],[139,0.3443],[142,0.3445],[143,0.3399],[145,0.34],[146,0.3357],[149,0.3359],[150,0.3317],[154,0.3319],[155,0.3276],[165,0.3279],[166,0.325],[168,0.325],[169,0.3244],[170,0.3215],[172,0.3209],[173,0.318],[180,0.3172],[181,0.3159],[186,0.3159],[187,0.3219],[193,0.3219],[194,0.3229],[197,0.3226],[198,0.3236],[208,0.3232],[209,0.3223],[212,0.3223],[213,0.3285],[215,0.3282],[216,0.3344],[217,0.3344],[218,0.3405],[219,0.3407],[221,0.3532],[222,0.3536],[224,0.3665],[225,0.3702],[226,0.3763],[227,0.3816],[228,0.3877],[229,0.393],[230,0.3938],[231,0.3991],[232,0.4052],[233,0.4052],[234,0.4061],[235,0.4113],[236,0.4122],[237,0.4122],[238,0.413],[239,0.4078],[240,0.4078],[243,0.3978],[244,0.3979],[247,0.388],[248,0.3881],[249,0.3847],[250,0.3847],[251,0.3814],[255,0.3813],[256,0.3846],[257,0.3834],[258,0.3853],[259,0.3853],[260,0.3873],[261,0.3873],[262,0.3892],[263,0.3892],[264,0.3912],[265,0.3912],[266,0.3931],[267,0.3931],[268,0.3953],[271,0.3948],[272,0.397],[277,0.3967],[278,0.394],[285,0.3935],[286,0.3907],[287,0.3907],[288,0.3882],[289,0.3866],[294,0.3705],[295,0.3707],[296,0.3676],[297,0.3676],[298,0.3645],[304,0.3648],[305,0.3664],[307,0.3664],[308,0.3721],[312,0.3719],[313,0.3776],[317,0.3774],[318,0.3831],[320,0.3829],[354,0.4243]]
       }
       neck_02_ctrl.Location.Z {
        type=MovieSceneScriptingFloatChannel
        default=0
        thinned_from=322
        keys [[0,-0.1907],[1,-0.192],[2,-0.1939],[5,-0.1939],[6,-0.1953],[8,-0.1953],[9,-0.1972],[13,-0.1972],[14,-0.1986],[15,-0.2005],[19,-0.2008],[20,-0.2044],[23,-0.2044],[24,-0.208],[28,-0.208],[29,-0.2116],[32,-0.2116],[33,-0.2112],[34,-0.2115],[35,-0.2125],[38,-0.2125],[39,-0.2138],[44,-0.2138],[45,-0.2148],[64,-0.215],[65,-0.2131],[66,-0.2131],[67,-0.2121],[68,-0.2121],[69,-0.2107],[74,-0.2107],[76,-0.2082],[78,-0.2082],[79,-0.2057],[80,-0.2057],[82,-0.204],[86,-0.1997],[89,-0.1979],[93,-0.1981],[95,-0.1973],[96,-0.1961],[98,-0.1953],[99,-0.1922],[100,-0.1918],[101,-0.1883],[102,-0.1883],[103,-0.1848],[104,-0.1848],[106,-0.1782],[107,-0.1778],[108,-0.1746],[110,-0.1743],[111,-0.1711],[112,-0.1715],[113,-0.1713],[114,-0.1699],[116,-0.1697],[118,-0.1682],[119,-0.1687],[122,-0.1685],[123,-0.169],[124,-0.1683],[125,-0.1695],[126,-0.1693],[127,-0.1701],[128,-0.1698],[130,-0.1709],[131,-0.1709],[132,-0.1698],[134,-0.1698],[135,-0.1703],[138,-0.1703],[139,-0.1688],[140,-0.1688],[141,-0.1683],[142,-0.1683],[143,-0.1668],[145,-0.1666],[146,-0.1652],[147,-0.1647],[149,-0.1647],[150,-0.1633],[151,-0.1633],[152,-0.1628],[154,-0.1628],[155,-0.1614],[156,-0.1614],[157,-0.1609],[165,-0.1606],[166,-0.1597],[168,-0.1597],[169,-0.1614],[170,-0.1604],[171,-0.1604],[172,-0.1621],[173,-0.1611],[174,-0.1611],[175,-0.1628],[176,-0.1628],[177,-0.1632],[180,-0.1632],[181,-0.1674],[186,-0.1674],[187,-0.1694],[193,-0.1693],[195,-0.1702],[197,-0.1702],[199,-0.1711],[204,-0.1711],[205,-0.1717],[212,-0.1717],[213,-0.1738],[214,-0.1744],[215,-0.1744],[216,-0.1765],[217,-0.1765],[218,-0.1785],[219,-0.1779],[220,-0.1799],[221,-0.1813],[222,-0.18],[223,-0.1814],[224,-0.1822],[225,-0.1822],[226,-0.1814],[227,-0.1832],[228,-0.1824],[229,-0.1841],[230,-0.1816],[231,-0.1834],[232,-0.1826],[233,-0.1826],[234,-0.18],[235,-0.1818],[236,-0.1793],[237,-0.1793],[238,-0.1767],[239,-0.175],[240,-0.175],[243,-0.1706],[244,-0.1703],[247,-0.1662],[248,-0.1659],[249,-0.1648],[250,-0.1648],[251,-0.1633],[252,-0.1633],[253,-0.1637],[255,-0.1637],[256,-0.1651],[257,-0.1649],[258,-0.1663],[259,-0.1663],[260,-0.1678],[261,-0.1678],[262,-0.1693],[263,-0.1693],[264,-0.1708],[265,-0.1708],[266,-0.1723],[267,-0.1723],[269,-0.1737],[270,-0.1737],[272,-0.1752],[274,-0.1753],[275,-0.1761],[277,-0.1761],[278,-0.1751],[279,-0.1759],[282,-0.1759],[283,-0.1766],[285,-0.1766],[286,-0.1757],[287,-0.1757],[288,-0.174],[289,-0.1734],[290,-0.1715],[291,-0.1704],[293,-0.1666],[294,-0.1655],[295,-0.1647],[296,-0.1628],[297,-0.1628],[298,-0.1609],[301,-0.1609],[302,-0.1601],[304,-0.1601],[305,-0.1606],[307,-0.1606],[308,-0.1625],[309,-0.163],[312,-0.163],[313,-0.1649],[314,-0.1649],[315,-0.1655],[317,-0.1655],[318,-0.1674],[319,-0.168],[320,-0.168],[354,-0.1907]]
       }
       neck_02_ctrl.Rotation.X {
        type=MovieSceneScriptingFloatChannel
        default=0
        thinned_from=322
        keys [[0,18.1252],[90,18.129],[93,18.0295],[96,17.7748],[101,17.105],[109,15.871],[112,15.512],[117,15.012],[122,14.5805],[130,13.9664],[137,13.5426],[146,13.2657],[152,13.162],[165,13.1112],[167,13.0439],[169,12.8465],[170,12.7281],[173,12.2538],[178,11.3768],[180,11.0798],[182,10.8555],[183,10.7797],[184,10.7269],[217,10.8067],[218,10.86],[219,10.9883],[220,11.1845],[221,11.4449],[222,11.7626],[223,12.1232],[224,12.5267],[225,12.9569],[227,13.8866],[231,15.7993],[232,16.2389],[233,16.6443],[234,17.0055],[235,17.3275],[236,17.5854],[237,17.7791],[238,17.8987],[239,17.939],[320,18.0336],[354,18.1252]]
       }
       neck_02_ctrl.Rotation.Y {
        type=MovieSceneScriptingFloatChannel
        default=0
        thinned_from=322
        keys [[0,-9.4595],[220,-9.2001],[235,-9.4043],[320,-9.4688],[354,-9.4595]]
       }
       neck_02_ctrl.Rotation.Z {
        type=MovieSceneScriptingFloatChannel
        default=0
        thinned_from=322
        keys [[0,3.1832],[1,3.1729],[27,3.1954],[58,3.1193],[86,3.2079],[91,3.2537],[98,3.3375],[106,3.364],[153,3.3295],[165,3.3375],[175,3.2839],[211,3.1992],[217,3.216],[219,3.2353],[222,3.2899],[229,3.4674],[232,3.5212],[234,3.5403],[237,3.5434],[241,3.5095],[251,3.3705],[293,2.999],[298,2.9896],[304,3.0182],[315,3.1089],[320,3.1315],[354,3.1832]]
       }
       neck_02_ctrl.Scale.X {type=MovieSceneScriptingFloatChannel, default=1, const [1,322]}
       neck_02_ctrl.Scale.Y {type=MovieSceneScriptingFloatChannel, default=1, const [1,322]}
       neck_02_ctrl.Scale.Z {type=MovieSceneScriptingFloatChannel, default=1, const [1,322]}
       head_ik_ctrl.Location.X {
        type=MovieSceneScriptingFloatChannel
        default=0
        thinned_from=322
        keys [[0,1.2969],[2,1.306],[5,1.3124],[8,1.3257],[19,1.3665],[20,1.3737],[25,1.3898],[26,1.3951],[27,1.3973],[28,1.403],[34,1.4137],[35,1.4178],[39,1.4214],[40,1.4262],[44,1.4293],[45,1.4333],[60,1.4379],[61,1.4346],[64,1.4351],[68,1.431],[69,1.4265],[74,1.4256],[75,1.4216],[78,1.4217],[79,1.418],[87,1.4181],[95,1.4467],[98,1.4406],[100,1.4279],[101,1.4179],[102,1.4113],[103,1.4004],[104,1.3942],[105,1.382],[108,1.3597],[112,1.3421],[120,1.3277],[131,1.3169],[135,1.3084],[136,1.3021],[159,1.2796],[166,1.2824],[167,1.2848],[170,1.2822],[171,1.2837],[173,1.2787],[175,1.2762],[179,1.2636],[180,1.2546],[183,1.2461],[184,1.2498],[195,1.2525],[196,1.2559],[199,1.2585],[204,1.2685],[205,1.2728],[212,1.2858],[213,1.2824],[216,1.2892],[217,1.2859],[218,1.2905],[219,1.2816],[220,1.2471],[221,1.2003],[222,1.1351],[223,1.0565],[224,0.9625],[225,0.8579],[226,0.7423],[227,0.6164],[228,0.4872],[229,0.3599],[230,0.2279],[231,0.1022],[232,-0.0171],[233,-0.1301],[234,-0.223],[235,-0.3035],[236,-0.3688],[237,-0.4093],[238,-0.4257],[239,-0.4255],[240,-0.416],[241,-0.3992],[243,-0.3475],[244,-0.3135],[245,-0.2767],[247,-0.183],[248,-0.1342],[249,-0.081],[250,-0.024],[252,0.0992],[253,0.1612],[254,0.2254],[255,0.2914],[256,0.3566],[258,0.4888],[259,0.5567],[261,0.687],[263,0.811],[265,0.9258],[267,1.0263],[268,1.0699],[269,1.111],[270,1.1459],[271,1.1784],[272,1.2038],[273,1.2255],[275,1.2544],[276,1.2577],[278,1.258],[279,1.2632],[283,1.2636],[284,1.2687],[292,1.2716],[293,1.2703],[294,1.2725],[297,1.2703],[298,1.273],[305,1.2809],[320,1.3159],[354,1.2969]]
       }
       head_ik_ctrl.Location.Y {
        type=MovieSceneScriptingFloatChannel
        default=0
        thinned_from=322
        keys [[0,0.8043],[1,0.8008],[2,0.8048],[6,0.7899],[7,0.7799],[8,0.7745],[9,0.774],[13,0.7486],[14,0.7378],[15,0.7373],[19,0.7122],[20,0.7115],[21,0.705],[22,0.6915],[23,0.6854],[24,0.6859],[27,0.6706],[28,0.667],[29,0.6696],[30,0.6662],[33,0.6613],[34,0.6617],[38,0.6587],[39,0.6597],[40,0.6582],[49,0.6552],[63,0.6486],[64,0.6491],[65,0.6463],[66,0.6457],[67,0.642],[69,0.6411],[70,0.6422],[74,0.6397],[75,0.6411],[76,0.6368],[78,0.6381],[79,0.6359],[80,0.6378],[81,0.6372],[83,0.6382],[85,0.6355],[86,0.6304],[87,0.6235],[88,0.6144],[89,0.602],[90,0.5876],[91,0.5698],[92,0.5472],[93,0.5191],[94,0.4862],[96,0.4108],[98,0.3321],[99,0.2922],[100,0.2558],[101,0.2184],[103,0.1528],[104,0.1238],[105,0.0887],[106,0.063],[107,0.0434],[108,0.0225],[109,0.0139],[110,0.0005],[111,-0.0127],[112,-0.0155],[113,-0.0166],[114,-0.0234],[115,-0.0281],[117,-0.0313],[118,-0.0346],[119,-0.0364],[120,-0.0362],[121,-0.0422],[122,-0.0493],[123,-0.0546],[124,-0.0677],[125,-0.0746],[127,-0.1009],[128,-0.1172],[129,-0.1304],[130,-0.1447],[131,-0.1597],[132,-0.1772],[133,-0.1915],[134,-0.2038],[135,-0.2155],[136,-0.2257],[137,-0.2478],[138,-0.256],[139,-0.269],[140,-0.2761],[142,-0.2882],[143,-0.2979],[144,-0.3026],[145,-0.3083],[146,-0.3167],[147,-0.3219],[148,-0.3252],[149,-0.3293],[150,-0.3367],[152,-0.3433],[154,-0.3486],[155,-0.3536],[156,-0.3557],[158,-0.3576],[160,-0.3573],[161,-0.3565],[165,-0.3557],[166,-0.3583],[167,-0.3647],[169,-0.3802],[170,-0.394],[171,-0.4098],[172,-0.4243],[173,-0.443],[174,-0.4591],[179,-0.5461],[181,-0.5771],[183,-0.6002],[184,-0.608],[185,-0.6105],[186,-0.6106],[187,-0.6044],[192,-0.6018],[194,-0.5984],[196,-0.5977],[200,-0.5902],[201,-0.5901],[205,-0.5795],[209,-0.5647],[211,-0.5552],[212,-0.5537],[213,-0.5407],[214,-0.5343],[215,-0.5306],[216,-0.5127],[217,-0.5057],[218,-0.4832],[219,-0.4528],[220,-0.4002],[221,-0.3387],[222,-0.2698],[223,-0.1895],[224,-0.1036],[225,-0.0221],[226,0.063],[227,0.1449],[228,0.2229],[229,0.2944],[230,0.3534],[231,0.4116],[232,0.4656],[233,0.5031],[235,0.5674],[236,0.5879],[237,0.6041],[238,0.6159],[239,0.6121],[240,0.614],[241,0.6118],[242,0.6119],[243,0.6149],[244,0.6207],[245,0.6232],[246,0.6286],[248,0.6446],[249,0.6503],[250,0.6598],[251,0.6663],[252,0.6781],[255,0.7036],[256,0.7146],[257,0.7207],[258,0.7292],[259,0.7352],[260,0.7428],[269,0.7896],[272,0.803],[275,0.812],[276,0.8114],[277,0.8132],[278,0.8121],[279,0.8141],[280,0.8121],[285,0.8186],[286,0.8143],[287,0.816],[289,0.8147],[290,0.8112],[291,0.813],[293,0.805],[294,0.8063],[295,0.8053],[296,0.8007],[297,0.7989],[298,0.7993],[302,0.7874],[304,0.7789],[305,0.7758],[307,0.7667],[308,0.767],[309,0.7582],[310,0.7523],[311,0.7483],[312,0.7423],[313,0.7423],[314,0.7336],[315,0.7281],[317,0.7191],[318,0.72],[319,0.7122],[320,0.708],[354,0.8043]]
       }
       head_ik_ctrl.Location.Z {
        type=MovieSceneScriptingFloatChannel
        default=0
        thinned_from=322
        keys [[0,-0.1656],[1,-0.1584],[2,-0.1537],[3,-0.1504],[5,-0.1419],[6,-0.1392],[7,-0.1341],[8,-0.1308],[9,-0.1279],[11,-0.1176],[12,-0.1107],[13,-0.1027],[14,-0.0922],[15,-0.0824],[16,-0.0717],[17,-0.0606],[18,-0.0493],[19,-0.0384],[20,-0.0292],[21,-0.0189],[22,-0.0076],[23,0.001],[24,0.0065],[25,0.0126],[26,0.0175],[27,0.021],[28,0.024],[29,0.025],[30,0.0233],[31,0.0254],[32,0.0266],[33,0.0272],[34,0.0261],[35,0.026],[36,0.0244],[37,0.0243],[38,0.0223],[40,0.0192],[41,0.0184],[42,0.016],[43,0.0152],[44,0.0128],[45,0.0113],[46,0.0105],[47,0.008],[48,0.0071],[49,0.0048],[51,0.0036],[52,0.0023],[53,0.0018],[56,-0.0013],[59,-0.0052],[60,-0.0067],[61,-0.0051],[63,-0.008],[64,-0.0098],[65,-0.0102],[66,-0.0117],[67,-0.0118],[68,-0.0133],[70,-0.0144],[72,-0.0172],[73,-0.0172],[74,-0.0185],[75,-0.018],[76,-0.0192],[77,-0.0193],[78,-0.0208],[79,-0.0198],[81,-0.0206],[83,-0.0202],[84,-0.0202],[85,-0.0197],[86,-0.0188],[87,-0.0182],[89,-0.0153],[90,-0.014],[91,-0.012],[92,-0.0098],[93,-0.0068],[94,-0.0034],[95,0.0005],[96,0.0042],[98,0.0101],[99,0.0103],[100,0.0116],[101,0.0099],[102,0.0095],[103,0.0064],[104,0.0029],[105,0.001],[107,-0.0085],[108,-0.0151],[109,-0.023],[110,-0.0287],[111,-0.036],[112,-0.0444],[114,-0.0601],[115,-0.0663],[116,-0.0745],[117,-0.0805],[118,-0.0881],[119,-0.0963],[120,-0.1018],[121,-0.1087],[122,-0.112],[124,-0.12],[125,-0.123],[126,-0.1252],[127,-0.1285],[128,-0.1294],[129,-0.1313],[130,-0.1325],[132,-0.1336],[134,-0.1342],[135,-0.1349],[136,-0.1341],[137,-0.1309],[138,-0.1284],[140,-0.1253],[141,-0.1232],[142,-0.1214],[143,-0.1191],[144,-0.1171],[145,-0.1146],[147,-0.1083],[148,-0.1058],[149,-0.103],[152,-0.0941],[154,-0.0893],[155,-0.0866],[156,-0.0849],[157,-0.0825],[158,-0.081],[159,-0.08],[160,-0.0792],[161,-0.0776],[163,-0.0769],[164,-0.0768],[166,-0.0769],[167,-0.0827],[168,-0.0813],[169,-0.0819],[170,-0.0797],[171,-0.084],[172,-0.0832],[173,-0.0804],[174,-0.0849],[175,-0.0843],[176,-0.0818],[177,-0.0806],[178,-0.0783],[179,-0.0765],[180,-0.074],[181,-0.0826],[182,-0.0818],[183,-0.0812],[184,-0.0815],[185,-0.0825],[186,-0.084],[187,-0.0863],[189,-0.0892],[193,-0.0941],[194,-0.0951],[195,-0.0963],[198,-0.0985],[199,-0.0998],[200,-0.1006],[201,-0.1097],[204,-0.1108],[205,-0.1117],[208,-0.1123],[209,-0.1132],[211,-0.113],[212,-0.112],[213,-0.1113],[214,-0.1116],[215,-0.1104],[216,-0.0983],[217,-0.0964],[218,-0.0978],[219,-0.0984],[220,-0.086],[221,-0.0861],[222,-0.0856],[223,-0.0736],[224,-0.0748],[225,-0.0729],[226,-0.0719],[227,-0.073],[228,-0.075],[230,-0.0861],[231,-0.095],[232,-0.1044],[233,-0.1159],[234,-0.1264],[235,-0.1391],[236,-0.1498],[237,-0.1613],[238,-0.1695],[239,-0.1753],[240,-0.1815],[241,-0.1853],[242,-0.1873],[243,-0.1997],[245,-0.2038],[246,-0.2048],[247,-0.2159],[248,-0.2174],[250,-0.2188],[251,-0.2182],[252,-0.2283],[253,-0.2281],[254,-0.2272],[255,-0.2254],[256,-0.2247],[257,-0.2221],[259,-0.2187],[260,-0.2159],[261,-0.2112],[262,-0.2095],[263,-0.2042],[264,-0.2026],[265,-0.1981],[266,-0.1975],[267,-0.1941],[268,-0.1938],[269,-0.1929],[270,-0.1936],[271,-0.193],[272,-0.1943],[273,-0.194],[275,-0.1916],[276,-0.1965],[278,-0.1935],[279,-0.1924],[280,-0.197],[282,-0.1949],[283,-0.1947],[284,-0.1927],[285,-0.1918],[286,-0.1964],[287,-0.1959],[288,-0.1941],[289,-0.193],[290,-0.1923],[291,-0.1934],[292,-0.1927],[293,-0.1905],[294,-0.1914],[295,-0.1901],[296,-0.1878],[297,-0.1874],[298,-0.1876],[299,-0.1853],[300,-0.1842],[301,-0.1817],[302,-0.1797],[303,-0.1785],[304,-0.1758],[305,-0.174],[306,-0.1728],[307,-0.1681],[308,-0.1682],[309,-0.1659],[310,-0.1648],[311,-0.1596],[312,-0.158],[313,-0.1578],[314,-0.1549],[315,-0.1537],[316,-0.1482],[317,-0.1464],[318,-0.1458],[319,-0.1432],[320,-0.1415],[354,-0.1656]]
       }
       head_ik_ctrl.Rotation.X {
        type=MovieSceneScriptingFloatChannel
        default=0
        thinned_from=322
        keys [[0,-5.7506],[1,-5.3693],[2,-5.2561],[3,-5.176],[4,-5.1261],[5,-5.1044],[6,-5.1087],[7,-5.1392],[8,-5.1867],[9,-5.2626],[10,-5.3614],[11,-5.4735],[12,-5.6053],[14,-5.9037],[18,-6.5295],[20,-6.796],[21,-6.9138],[23,-7.0911],[25,-7.2042],[27,-7.2423],[29,-7.2095],[30,-7.1671],[32,-7.0371],[33,-6.9529],[35,-6.7421],[37,-6.5027],[45,-5.4865],[47,-5.2714],[50,-5],[52,-4.8467],[53,-4.7907],[54,-4.7468],[55,-4.6932],[59,-4.5697],[63,-4.4867],[66,-4.4509],[68,-4.445],[70,-4.4731],[71,-4.5143],[72,-4.5662],[73,-4.6379],[74,-4.7273],[75,-4.8445],[76,-4.9927],[77,-5.1722],[78,-5.4146],[79,-5.734],[80,-6.1024],[81,-6.5142],[82,-6.9631],[86,-8.8483],[87,-9.2909],[88,-9.6959],[89,-10.0457],[91,-10.6073],[93,-11.1363],[94,-11.3839],[95,-11.6016],[96,-11.7905],[97,-11.9385],[98,-12.0133],[99,-12.0024],[100,-11.9288],[101,-11.795],[102,-11.6122],[103,-11.3861],[106,-10.5575],[108,-9.9856],[111,-9.2179],[113,-8.7376],[115,-8.3226],[116,-8.1505],[117,-8.0299],[118,-7.9572],[120,-7.9414],[121,-7.9886],[123,-8.1901],[125,-8.4954],[126,-8.6691],[129,-9.2717],[132,-9.8981],[134,-10.2781],[135,-10.4496],[136,-10.595],[137,-10.7074],[138,-10.7914],[140,-10.8528],[144,-10.8555],[157,-10.5182],[162,-10.2885],[163,-10.2673],[164,-10.2178],[165,-10.1048],[166,-9.9324],[168,-9.4927],[170,-8.9763],[171,-8.7342],[172,-8.5223],[173,-8.3549],[174,-8.2474],[175,-8.2024],[177,-8.2583],[180,-8.4593],[181,-8.5421],[184,-8.73],[187,-8.8668],[188,-8.8953],[189,-8.8513],[191,-8.8697],[192,-8.8629],[194,-8.8012],[196,-8.6885],[200,-8.3724],[204,-8.0075],[206,-7.851],[208,-7.7439],[209,-7.7102],[210,-7.6924],[211,-7.7072],[212,-7.7866],[213,-7.8931],[215,-8.1604],[216,-8.3321],[217,-8.5437],[218,-8.7959],[220,-9.3612],[222,-9.9675],[223,-10.2557],[224,-10.5224],[225,-10.7637],[226,-10.9745],[227,-11.1566],[228,-11.3075],[229,-11.4346],[230,-11.5334],[231,-11.6017],[232,-11.6351],[233,-11.6355],[234,-11.6012],[235,-11.5107],[236,-11.3848],[237,-11.215],[238,-11.0103],[239,-10.7527],[240,-10.4345],[241,-10.0553],[242,-9.6252],[243,-9.1451],[248,-6.4636],[249,-5.9437],[250,-5.4479],[251,-4.9754],[252,-4.5306],[253,-4.1182],[254,-3.7351],[255,-3.3834],[256,-3.0497],[257,-2.7342],[258,-2.4255],[260,-1.7893],[261,-1.4648],[263,-0.832],[264,-0.5206],[265,-0.2159],[266,0.0844],[267,0.3797],[268,0.6694],[269,0.9634],[271,1.5318],[273,2.0735],[274,2.3302],[275,2.5755],[276,2.809],[280,3.6968],[282,4.1183],[284,4.5017],[286,4.8405],[288,5.1551],[290,5.4508],[292,5.7196],[293,5.8251],[294,5.9098],[295,5.9668],[296,5.9944],[297,6.0051],[298,5.9897],[299,5.9418],[300,5.8699],[301,5.7766],[302,5.6661],[303,5.5381],[304,5.3923],[305,5.2354],[306,5.0673],[311,4.1445],[313,3.798],[314,3.6371],[315,3.4907],[316,3.3567],[317,3.2451],[318,3.1527],[319,3.0773],[320,3.0159],[354,-5.7506]]
       }
       head_ik_ctrl.Rotation.Y {
        type=MovieSceneScriptingFloatChannel
        default=0
        thinned_from=322
        keys [[0,-4.3117],[1,-3.4371],[2,-3.0805],[3,-2.7387],[4,-2.4155],[5,-2.1057],[6,-1.8132],[7,-1.5327],[8,-1.2654],[9,-1.0114],[10,-0.7351],[11,-0.4007],[12,0.0085],[13,0.4682],[14,0.9722],[15,1.5186],[16,2.0984],[17,2.7007],[20,4.5803],[21,5.1965],[22,5.7862],[23,6.3423],[24,6.8525],[25,7.2913],[26,7.6628],[27,7.9697],[28,8.2364],[29,8.4664],[30,8.6489],[31,8.7942],[32,8.9023],[33,8.9823],[36,9.1164],[40,9.196],[47,9.1587],[51,9.0867],[52,9.0435],[53,9.0211],[56,8.8623],[59,8.6379],[68,7.7767],[73,7.2183],[75,6.9602],[77,6.665],[78,6.484],[79,6.265],[80,6.0198],[82,5.4756],[83,5.1901],[85,4.6003],[86,4.3351],[87,4.1482],[88,4.0503],[89,4.0319],[90,4.1458],[91,4.3656],[92,4.6501],[93,4.9799],[95,5.6639],[96,5.9786],[97,6.2358],[98,6.4087],[99,6.4781],[100,6.4535],[101,6.3643],[102,6.1971],[103,5.9501],[104,5.6313],[105,5.2144],[106,4.6984],[107,4.1038],[108,3.4618],[109,2.7769],[110,2.0674],[112,0.6307],[113,-0.0813],[114,-0.7708],[115,-1.4526],[116,-2.12],[117,-2.7617],[118,-3.3663],[119,-3.9127],[120,-4.402],[121,-4.817],[122,-5.1478],[123,-5.4115],[124,-5.6102],[125,-5.7475],[126,-5.8119],[127,-5.8239],[128,-5.7991],[129,-5.7535],[133,-5.4545],[134,-5.3875],[136,-5.2828],[139,-5.1634],[144,-4.995],[152,-4.715],[155,-4.6456],[157,-4.6271],[159,-4.6353],[162,-4.6954],[163,-4.7229],[164,-4.7848],[165,-4.9003],[166,-5.049],[167,-5.2282],[171,-6.0489],[172,-6.2447],[173,-6.4136],[174,-6.5397],[175,-6.6229],[176,-6.6775],[177,-6.7076],[178,-6.686],[179,-6.6401],[181,-6.5],[185,-6.1695],[186,-6.1048],[191,-5.6763],[193,-5.5239],[195,-5.392],[198,-5.2392],[200,-5.1798],[202,-5.1564],[208,-5.1896],[209,-5.1836],[210,-5.1536],[211,-5.0176],[212,-4.653],[213,-4.2348],[214,-3.7891],[215,-3.3346],[216,-2.8723],[217,-2.3949],[218,-1.9006],[219,-1.4337],[220,-1.0046],[221,-0.6509],[222,-0.3697],[223,-0.1862],[224,-0.106],[225,-0.1058],[226,-0.1722],[227,-0.3065],[228,-0.5024],[229,-0.7501],[230,-1.0432],[231,-1.3796],[232,-1.7414],[233,-2.1239],[234,-2.5003],[235,-2.8859],[236,-3.289],[237,-3.7196],[238,-4.1787],[239,-4.6524],[240,-5.1406],[242,-6.1383],[244,-7.1103],[245,-7.5836],[246,-8.0403],[247,-8.473],[248,-8.8732],[249,-9.2363],[250,-9.5343],[251,-9.7693],[252,-9.9176],[253,-9.9806],[254,-9.9792],[255,-9.9214],[256,-9.8208],[257,-9.6805],[258,-9.5156],[259,-9.3257],[263,-8.4573],[264,-8.279],[265,-8.1201],[266,-7.9959],[267,-7.9051],[268,-7.8443],[269,-7.8206],[273,-7.7895],[285,-7.4721],[286,-7.4993],[287,-7.613],[288,-7.8037],[289,-8.0413],[290,-8.3258],[292,-8.9448],[293,-9.2363],[294,-9.493],[295,-9.6999],[296,-9.8474],[297,-9.9742],[298,-10.0536],[299,-10.0907],[300,-10.0923],[301,-10.0696],[302,-10.0205],[303,-9.95],[305,-9.7443],[307,-9.4604],[315,-8.0636],[317,-7.7477],[320,-7.3434],[354,-4.3117]]
       }
       head_ik_ctrl.Rotation.Z {
        type=MovieSceneScriptingFloatChannel
        default=0
        thinned_from=322
        keys [[0,-1.2578],[1,-0.6105],[2,-0.3839],[3,-0.1608],[4,0.0568],[5,0.2686],[6,0.471],[7,0.6648],[8,0.851],[10,1.1988],[12,1.5297],[13,1.6877],[16,2.1206],[17,2.2559],[18,2.3826],[19,2.4969],[20,2.5887],[21,2.6684],[22,2.7273],[23,2.7742],[24,2.7933],[25,2.7916],[26,2.7742],[27,2.7435],[28,2.7033],[29,2.6509],[31,2.5284],[34,2.3183],[37,2.1252],[40,1.9451],[42,1.8342],[44,1.7384],[46,1.654],[47,1.6183],[48,1.5869],[49,1.56],[51,1.5209],[53,1.5031],[55,1.5073],[56,1.5165],[57,1.5309],[59,1.5706],[61,1.6218],[65,1.7493],[68,1.86],[69,1.8707],[70,1.8611],[71,1.8227],[72,1.7484],[73,1.6283],[74,1.4608],[75,1.2294],[76,0.9184],[77,0.4944],[78,-0.1238],[79,-0.9889],[80,-2.0326],[81,-3.2403],[82,-4.5462],[83,-5.9166],[84,-7.3344],[85,-8.7993],[86,-10.2425],[87,-11.6077],[88,-12.8335],[89,-13.874],[90,-14.6115],[91,-15.1795],[92,-15.6492],[93,-16.023],[94,-16.3106],[95,-16.5001],[96,-16.596],[97,-16.5822],[98,-16.4447],[99,-16.1859],[100,-15.8311],[101,-15.3622],[102,-14.8019],[103,-14.1524],[104,-13.4096],[106,-11.8296],[108,-10.3101],[109,-9.5825],[110,-8.8932],[111,-8.2422],[112,-7.6351],[113,-7.0779],[114,-6.5848],[115,-6.1333],[116,-5.7094],[117,-5.3274],[118,-4.9939],[119,-4.7095],[120,-4.4838],[121,-4.3182],[122,-4.2231],[123,-4.1824],[124,-4.1759],[125,-4.2076],[126,-4.289],[127,-4.4169],[128,-4.5994],[129,-4.8173],[131,-5.2975],[132,-5.523],[133,-5.7198],[134,-5.8937],[135,-6.0403],[136,-6.1566],[137,-6.2442],[138,-6.3039],[139,-6.3479],[141,-6.3673],[142,-6.3533],[144,-6.2862],[146,-6.17],[152,-5.7242],[154,-5.6026],[155,-5.5597],[156,-5.5304],[157,-5.5176],[158,-5.523],[159,-5.5482],[160,-5.5933],[161,-5.6646],[162,-5.7626],[163,-5.8912],[164,-6.0813],[165,-6.3647],[166,-6.7271],[167,-7.1446],[168,-7.6054],[171,-9.0778],[172,-9.5394],[173,-9.9542],[174,-10.3073],[175,-10.5927],[176,-10.8257],[178,-11.2262],[180,-11.5207],[182,-11.7265],[183,-11.7999],[184,-11.8465],[186,-11.8568],[192,-11.5147],[204,-11.0205],[205,-10.9302],[206,-10.8169],[207,-10.6671],[208,-10.4861],[209,-10.2556],[210,-9.9384],[211,-9.3679],[212,-8.2434],[213,-6.8972],[214,-5.3531],[215,-3.6502],[216,-1.8416],[217,0.0395],[218,1.9729],[219,3.9248],[220,5.8923],[221,7.8788],[224,13.912],[225,15.8402],[226,17.6879],[227,19.4175],[228,21.0126],[229,22.4473],[230,23.7064],[231,24.7562],[232,25.5815],[233,26.1584],[234,26.49],[235,26.6245],[236,26.5713],[237,26.3494],[238,25.9765],[239,25.4658],[240,24.8454],[241,24.1245],[242,23.3143],[243,22.4357],[245,20.526],[249,16.5258],[250,15.5548],[251,14.6219],[252,13.7426],[253,12.9224],[255,11.4337],[256,10.7128],[257,10.0176],[258,9.3517],[259,8.728],[260,8.1372],[261,7.5868],[262,7.0779],[263,6.6065],[264,6.1684],[265,5.7667],[266,5.409],[267,5.0942],[268,4.8228],[269,4.6119],[271,4.2446],[273,3.9158],[274,3.7668],[275,3.6316],[276,3.5067],[278,3.2866],[279,3.1845],[281,3.0082],[283,2.862],[284,2.7966],[285,2.7216],[286,2.5307],[287,2.0843],[288,1.4565],[289,0.6644],[290,-0.2326],[291,-1.2079],[292,-2.1868],[293,-3.1393],[294,-4.008],[295,-4.7403],[296,-5.2928],[297,-5.8068],[298,-6.1955],[299,-6.4896],[300,-6.7209],[301,-6.8867],[302,-7.0154],[303,-7.0923],[304,-7.1346],[305,-7.1343],[306,-7.0976],[307,-7.0203],[308,-6.9074],[309,-6.7751],[310,-6.6231],[312,-6.2867],[313,-6.1076],[314,-5.9085],[316,-5.4757],[318,-5.011],[320,-4.5686],[354,-1.2578]]
       }
       head_ik_ctrl.Scale.X {type=MovieSceneScriptingFloatChannel, default=1, const [1,322]}
       head_ik_ctrl.Scale.Y {type=MovieSceneScriptingFloatChannel, default=1, const [1,322]}
       head_ik_ctrl.Scale.Z {type=MovieSceneScriptingFloatChannel, default=1, const [1,322]}
       chest_ctrl.Location.X {
        type=MovieSceneScriptingFloatChannel
        default=0
        thinned_from=322
        keys [[0,-0.6631],[1,-0.747],[2,-0.7952],[3,-0.8344],[5,-0.9067],[7,-0.9749],[9,-1.0341],[10,-1.0674],[11,-1.1128],[12,-1.1743],[13,-1.2481],[14,-1.33],[15,-1.4199],[17,-1.6197],[18,-1.729],[19,-1.8281],[20,-1.9348],[22,-2.1268],[23,-2.2171],[25,-2.3579],[27,-2.4525],[28,-2.4851],[30,-2.5387],[31,-2.5602],[32,-2.5726],[35,-2.5865],[39,-2.5809],[48,-2.5197],[54,-2.4784],[58,-2.4381],[62,-2.3892],[63,-2.3722],[66,-2.3356],[72,-2.254],[79,-2.1748],[84,-2.127],[85,-2.1224],[87,-2.1299],[89,-2.1597],[91,-2.2062],[92,-2.2328],[94,-2.272],[95,-2.2799],[96,-2.2774],[97,-2.2629],[98,-2.2272],[99,-2.1684],[100,-2.0937],[101,-2.004],[102,-1.894],[103,-1.7783],[105,-1.5099],[107,-1.2074],[108,-1.053],[109,-0.8919],[110,-0.7484],[111,-0.5967],[112,-0.45],[113,-0.3154],[114,-0.1828],[115,-0.0552],[116,0.0577],[117,0.1671],[118,0.2733],[119,0.3601],[120,0.4378],[121,0.4997],[122,0.553],[123,0.5863],[124,0.6143],[125,0.6404],[126,0.6453],[127,0.6401],[128,0.6313],[129,0.6165],[132,0.5649],[133,0.5511],[134,0.5386],[136,0.5214],[138,0.5107],[139,0.5063],[140,0.504],[141,0.5004],[144,0.4957],[147,0.495],[160,0.4973],[161,0.4993],[162,0.498],[164,0.5022],[165,0.5057],[166,0.5081],[167,0.5118],[172,0.5244],[173,0.526],[174,0.5249],[176,0.5204],[177,0.5137],[178,0.5053],[179,0.489],[180,0.4751],[181,0.4643],[182,0.4503],[183,0.4351],[184,0.4242],[185,0.4156],[186,0.4084],[188,0.4006],[190,0.3858],[191,0.3826],[192,0.3775],[193,0.3756],[194,0.3728],[197,0.3673],[200,0.3671],[202,0.3699],[204,0.3758],[205,0.3823],[206,0.3859],[207,0.3921],[209,0.4096],[210,0.4176],[211,0.4321],[212,0.442],[213,0.4552],[214,0.4705],[215,0.4909],[216,0.5084],[217,0.5247],[218,0.5384],[219,0.5534],[220,0.5705],[221,0.5926],[222,0.6245],[223,0.6673],[224,0.7189],[225,0.7781],[226,0.8398],[227,0.9044],[228,0.9648],[229,1.0227],[230,1.0739],[231,1.1126],[232,1.1428],[233,1.1528],[234,1.1542],[235,1.1429],[236,1.1274],[238,1.0874],[240,1.0395],[242,0.9815],[246,0.8479],[247,0.8178],[248,0.7853],[250,0.7256],[252,0.6566],[253,0.6153],[254,0.5706],[255,0.5223],[256,0.4714],[257,0.4165],[258,0.3588],[259,0.2959],[260,0.2356],[261,0.1727],[262,0.108],[263,0.0526],[264,0.0017],[265,-0.0399],[266,-0.0795],[267,-0.1058],[268,-0.126],[269,-0.1381],[270,-0.1467],[271,-0.1572],[272,-0.1649],[273,-0.1709],[275,-0.189],[276,-0.1984],[277,-0.2108],[279,-0.2272],[280,-0.2363],[281,-0.2443],[282,-0.2532],[283,-0.2693],[284,-0.2811],[286,-0.3003],[287,-0.3091],[288,-0.3213],[289,-0.3305],[290,-0.3404],[291,-0.3552],[293,-0.3765],[294,-0.3862],[295,-0.4043],[296,-0.4165],[297,-0.4347],[298,-0.4482],[300,-0.4862],[301,-0.5098],[302,-0.53],[304,-0.5812],[305,-0.6044],[306,-0.6319],[307,-0.6556],[308,-0.6811],[309,-0.7087],[310,-0.7309],[311,-0.758],[312,-0.7801],[313,-0.805],[314,-0.826],[315,-0.85],[319,-0.921],[320,-0.9358],[354,-0.6631]]
       }
       chest_ctrl.Location.Y {
        type=MovieSceneScriptingFloatChannel
        default=0
        thinned_from=322
        keys [[0,-0.0278],[1,-0.0071],[2,-0.0025],[4,-0.004],[5,-0.0028],[6,-0.0028],[7,-0.0068],[8,-0.0087],[9,-0.0118],[10,-0.0201],[11,-0.031],[12,-0.0455],[13,-0.0626],[14,-0.0833],[15,-0.1085],[16,-0.1363],[17,-0.1658],[18,-0.1981],[19,-0.2321],[22,-0.3382],[23,-0.3707],[24,-0.3996],[25,-0.4271],[26,-0.4493],[28,-0.4887],[30,-0.5204],[31,-0.5327],[33,-0.5516],[34,-0.559],[35,-0.5631],[36,-0.562],[38,-0.5651],[40,-0.5639],[41,-0.5547],[42,-0.5528],[44,-0.5454],[45,-0.5397],[46,-0.5289],[47,-0.5241],[49,-0.5131],[50,-0.5045],[51,-0.4941],[52,-0.4848],[53,-0.4728],[54,-0.4623],[56,-0.4372],[57,-0.4182],[58,-0.4033],[60,-0.3713],[61,-0.3546],[63,-0.3191],[64,-0.2979],[65,-0.2809],[66,-0.2616],[67,-0.2466],[69,-0.2103],[70,-0.1925],[71,-0.177],[73,-0.1504],[74,-0.1374],[75,-0.1256],[76,-0.1148],[77,-0.107],[78,-0.0997],[79,-0.0913],[80,-0.0837],[81,-0.0758],[82,-0.0689],[83,-0.0613],[84,-0.0524],[85,-0.0424],[86,-0.0276],[87,-0.0056],[88,0.0251],[89,0.0639],[90,0.1128],[91,0.1651],[92,0.2274],[93,0.2915],[94,0.3583],[95,0.4273],[96,0.4952],[98,0.6208],[99,0.6797],[100,0.7336],[101,0.785],[102,0.8328],[103,0.874],[104,0.9078],[105,0.9364],[106,0.9535],[107,0.9664],[108,0.9684],[109,0.9604],[110,0.9467],[111,0.9306],[112,0.9115],[113,0.8901],[115,0.8372],[116,0.8045],[117,0.7584],[118,0.7065],[119,0.6466],[120,0.582],[121,0.5128],[123,0.3661],[124,0.2892],[125,0.2162],[126,0.1413],[127,0.0734],[128,0.0057],[129,-0.0559],[130,-0.1126],[131,-0.1631],[132,-0.209],[133,-0.25],[134,-0.2853],[135,-0.3118],[136,-0.3349],[137,-0.3499],[138,-0.3629],[139,-0.3687],[141,-0.3649],[142,-0.3575],[143,-0.3483],[144,-0.337],[146,-0.3107],[148,-0.275],[149,-0.2559],[150,-0.2393],[151,-0.2207],[152,-0.2041],[153,-0.1851],[154,-0.1699],[155,-0.1519],[156,-0.1395],[157,-0.1241],[158,-0.1103],[160,-0.0904],[161,-0.0786],[162,-0.0711],[163,-0.0611],[164,-0.0596],[165,-0.0565],[167,-0.056],[168,-0.0571],[169,-0.0624],[170,-0.0716],[171,-0.0772],[172,-0.0871],[173,-0.0983],[174,-0.114],[175,-0.1274],[176,-0.1416],[177,-0.1552],[178,-0.1709],[179,-0.1835],[180,-0.1992],[181,-0.2112],[182,-0.224],[183,-0.2392],[184,-0.2445],[185,-0.2602],[186,-0.2703],[187,-0.2823],[188,-0.2909],[189,-0.3017],[190,-0.3091],[191,-0.3138],[193,-0.3254],[194,-0.3272],[195,-0.3309],[197,-0.3348],[198,-0.3301],[199,-0.3311],[201,-0.3311],[202,-0.3298],[203,-0.3254],[204,-0.3244],[205,-0.3223],[206,-0.3221],[207,-0.3139],[208,-0.3122],[209,-0.3098],[210,-0.3048],[211,-0.3021],[212,-0.2936],[213,-0.2891],[215,-0.2721],[216,-0.2685],[217,-0.268],[218,-0.2635],[219,-0.2621],[220,-0.2624],[221,-0.2725],[223,-0.275],[224,-0.2791],[225,-0.2761],[227,-0.2578],[228,-0.2408],[229,-0.2213],[231,-0.1597],[232,-0.1213],[233,-0.0751],[234,-0.0213],[235,0.0308],[236,0.0923],[237,0.1486],[238,0.2155],[239,0.2792],[240,0.3553],[241,0.4208],[243,0.5707],[244,0.6428],[246,0.7951],[247,0.8661],[248,0.9427],[250,1.0821],[251,1.1443],[253,1.2623],[254,1.315],[255,1.3636],[256,1.4085],[259,1.5228],[261,1.5903],[264,1.6821],[266,1.732],[269,1.7935],[272,1.8363],[275,1.855],[276,1.8553],[278,1.8511],[280,1.8642],[284,1.8693],[286,1.8802],[292,1.9081],[300,1.9828],[313,2.129],[319,2.1762],[320,2.1804],[354,-0.0278]]
       }
       chest_ctrl.Location.Z {
        type=MovieSceneScriptingFloatChannel
        default=0
        thinned_from=322
        keys [[0,-0.0084],[1,-0.0111],[3,-0.0153],[4,-0.0186],[6,-0.023],[7,-0.0256],[8,-0.0268],[9,-0.029],[10,-0.032],[11,-0.0337],[12,-0.0383],[13,-0.0441],[14,-0.0494],[15,-0.0583],[16,-0.0665],[17,-0.077],[18,-0.0908],[19,-0.0972],[20,-0.1083],[21,-0.1189],[22,-0.1267],[23,-0.1417],[24,-0.1471],[25,-0.1592],[26,-0.1604],[27,-0.1688],[29,-0.1724],[30,-0.1786],[31,-0.175],[32,-0.1768],[33,-0.1791],[34,-0.1786],[36,-0.1781],[37,-0.1764],[39,-0.1742],[41,-0.171],[42,-0.1683],[46,-0.1608],[48,-0.1576],[49,-0.1552],[50,-0.1545],[52,-0.1495],[53,-0.1508],[54,-0.1526],[56,-0.146],[57,-0.1438],[58,-0.1402],[59,-0.1444],[60,-0.1404],[62,-0.1356],[63,-0.1378],[64,-0.135],[65,-0.131],[66,-0.1288],[67,-0.1322],[69,-0.128],[70,-0.1243],[71,-0.1303],[73,-0.1236],[74,-0.1223],[75,-0.1194],[76,-0.1184],[77,-0.1236],[78,-0.1213],[79,-0.1208],[80,-0.1187],[81,-0.1172],[82,-0.1152],[83,-0.1161],[84,-0.1149],[85,-0.114],[86,-0.1138],[87,-0.1167],[88,-0.1177],[89,-0.1205],[90,-0.1209],[91,-0.124],[92,-0.1248],[93,-0.1285],[94,-0.1315],[95,-0.13],[96,-0.1313],[97,-0.1295],[98,-0.1287],[99,-0.1217],[100,-0.1174],[101,-0.1099],[103,-0.0968],[104,-0.0872],[105,-0.0813],[106,-0.0709],[107,-0.0669],[108,-0.0587],[109,-0.0532],[110,-0.0502],[111,-0.0464],[112,-0.0432],[113,-0.0413],[114,-0.0391],[116,-0.0393],[117,-0.0373],[118,-0.0348],[119,-0.0369],[120,-0.0322],[121,-0.0346],[122,-0.0298],[123,-0.0257],[124,-0.0287],[125,-0.0245],[126,-0.0195],[127,-0.023],[128,-0.0189],[129,-0.0186],[130,-0.0171],[131,-0.0162],[132,-0.0149],[133,-0.0134],[135,-0.0095],[136,-0.0111],[137,-0.0099],[138,-0.0093],[139,-0.0098],[140,-0.0094],[141,-0.0114],[142,-0.0127],[143,-0.0124],[144,-0.0132],[145,-0.0126],[146,-0.0146],[147,-0.0143],[148,-0.0168],[149,-0.0164],[150,-0.0182],[151,-0.0177],[152,-0.0195],[153,-0.0193],[154,-0.0212],[155,-0.0211],[156,-0.0234],[157,-0.0227],[158,-0.0223],[159,-0.0249],[160,-0.0256],[161,-0.0255],[163,-0.0265],[164,-0.0273],[165,-0.0272],[166,-0.0286],[167,-0.0288],[168,-0.03],[169,-0.0302],[170,-0.0311],[172,-0.0305],[173,-0.0317],[174,-0.0324],[178,-0.0322],[180,-0.0315],[181,-0.0294],[182,-0.0293],[184,-0.0286],[185,-0.0285],[186,-0.0254],[187,-0.025],[188,-0.0253],[190,-0.0242],[191,-0.0221],[192,-0.0211],[193,-0.0215],[194,-0.0208],[195,-0.0194],[196,-0.0193],[197,-0.0189],[198,-0.0181],[199,-0.0176],[200,-0.0161],[201,-0.0162],[203,-0.0152],[205,-0.0153],[206,-0.0147],[207,-0.0138],[208,-0.0126],[209,-0.0126],[210,-0.0134],[211,-0.0134],[212,-0.013],[213,-0.0133],[214,-0.0129],[215,-0.0128],[216,-0.0133],[217,-0.0147],[220,-0.0159],[221,-0.0173],[222,-0.0184],[223,-0.0199],[224,-0.0237],[225,-0.0258],[226,-0.03],[227,-0.0348],[228,-0.0365],[229,-0.0424],[230,-0.0448],[231,-0.0455],[232,-0.047],[233,-0.0467],[234,-0.0473],[236,-0.0433],[237,-0.0417],[238,-0.0428],[239,-0.0389],[240,-0.04],[241,-0.0378],[242,-0.0423],[243,-0.0382],[244,-0.0436],[245,-0.0403],[246,-0.0465],[247,-0.0445],[248,-0.0472],[249,-0.0506],[250,-0.0499],[251,-0.0584],[252,-0.0582],[253,-0.0672],[254,-0.0673],[255,-0.0766],[256,-0.0771],[257,-0.0792],[258,-0.0856],[259,-0.0882],[260,-0.091],[261,-0.0979],[262,-0.101],[263,-0.1028],[264,-0.1062],[265,-0.1119],[266,-0.1151],[268,-0.1179],[270,-0.1197],[272,-0.1208],[273,-0.1201],[274,-0.1198],[275,-0.1206],[276,-0.1186],[277,-0.1186],[278,-0.118],[279,-0.1127],[280,-0.1121],[281,-0.1122],[282,-0.1117],[286,-0.113],[287,-0.1118],[288,-0.1122],[289,-0.1149],[290,-0.1159],[292,-0.1171],[293,-0.1184],[294,-0.1201],[295,-0.1226],[296,-0.1241],[297,-0.1277],[298,-0.1294],[299,-0.1333],[300,-0.135],[301,-0.1398],[303,-0.1461],[304,-0.151],[305,-0.1539],[307,-0.1624],[308,-0.1661],[309,-0.1711],[310,-0.1739],[311,-0.1785],[312,-0.1814],[313,-0.1859],[316,-0.1948],[318,-0.2001],[319,-0.2032],[320,-0.2055],[354,-0.0084]]
       }
       chest_ctrl.Rotation.X {
        type=MovieSceneScriptingFloatChannel
        default=0
        thinned_from=322
        keys [[0,-2.2727],[1,-2.1691],[2,-2.161],[3,-2.1623],[4,-2.1707],[5,-2.1873],[6,-2.2132],[7,-2.245],[8,-2.2861],[9,-2.3359],[10,-2.4025],[11,-2.4862],[12,-2.5871],[13,-2.7097],[14,-2.85],[15,-3.0109],[16,-3.1862],[17,-3.3739],[19,-3.7766],[21,-4.1904],[23,-4.5799],[24,-4.7584],[25,-4.9212],[26,-5.0655],[27,-5.1955],[28,-5.3119],[29,-5.4169],[31,-5.5828],[33,-5.6979],[35,-5.7596],[37,-5.7886],[39,-5.7947],[42,-5.7655],[49,-5.5897],[51,-5.5101],[54,-5.3486],[57,-5.1412],[60,-4.8945],[70,-3.9534],[72,-3.7909],[73,-3.7161],[75,-3.5855],[78,-3.4331],[84,-3.19],[86,-3.0935],[87,-3.0247],[88,-2.9141],[89,-2.7682],[90,-2.5962],[91,-2.4044],[92,-2.1944],[93,-1.9708],[94,-1.7333],[95,-1.4854],[96,-1.2312],[97,-0.9704],[98,-0.6947],[99,-0.395],[100,-0.0804],[101,0.243],[103,0.9017],[104,1.2222],[105,1.525],[106,1.8037],[107,2.049],[108,2.2528],[109,2.4094],[110,2.5274],[111,2.6212],[112,2.6917],[113,2.7345],[114,2.751],[115,2.7424],[116,2.6887],[117,2.5749],[118,2.4064],[119,2.1875],[120,1.9292],[121,1.632],[122,1.306],[123,0.9567],[124,0.5943],[125,0.2226],[126,-0.1545],[127,-0.5287],[128,-0.8958],[129,-1.2469],[130,-1.5765],[131,-1.8823],[132,-2.1582],[133,-2.4063],[134,-2.6218],[135,-2.8037],[136,-2.9519],[137,-3.0605],[138,-3.1305],[139,-3.1594],[140,-3.1568],[141,-3.1355],[142,-3.0982],[143,-3.0467],[144,-2.9808],[145,-2.9029],[146,-2.8122],[147,-2.7143],[152,-2.1822],[154,-1.9851],[155,-1.8917],[157,-1.7313],[158,-1.6581],[159,-1.5909],[160,-1.5305],[162,-1.4186],[163,-1.3729],[164,-1.3391],[165,-1.3203],[166,-1.3052],[167,-1.3078],[169,-1.334],[170,-1.3647],[171,-1.405],[172,-1.4556],[174,-1.5831],[175,-1.6567],[176,-1.7355],[179,-1.995],[181,-2.1724],[184,-2.4149],[186,-2.5639],[188,-2.6922],[189,-2.7501],[190,-2.8005],[191,-2.8452],[192,-2.8834],[193,-2.9153],[195,-2.9551],[197,-2.9759],[199,-2.9807],[201,-2.9736],[204,-2.9438],[208,-2.8767],[210,-2.8274],[212,-2.7594],[214,-2.6734],[215,-2.6254],[216,-2.5831],[217,-2.5505],[218,-2.5257],[219,-2.5083],[220,-2.4989],[222,-2.4917],[223,-2.4533],[224,-2.3905],[225,-2.3073],[226,-2.2083],[227,-2.0937],[228,-1.9645],[229,-1.8193],[230,-1.6606],[231,-1.4882],[232,-1.3019],[233,-1.1038],[234,-0.8884],[235,-0.6512],[236,-0.3991],[237,-0.132],[238,0.1478],[239,0.4406],[240,0.7448],[241,1.0618],[242,1.3931],[243,1.7368],[248,3.477],[249,3.8135],[250,4.1317],[251,4.4356],[252,4.7209],[253,4.9895],[254,5.2393],[255,5.4716],[256,5.6897],[258,6.0747],[260,6.4155],[262,6.7178],[264,6.9883],[266,7.2253],[268,7.4277],[270,7.594],[272,7.7171],[274,7.797],[276,7.8192],[280,7.8265],[289,7.9924],[297,8.32],[310,8.9921],[316,9.1785],[320,9.2522],[354,-2.2727]]
       }
       chest_ctrl.Rotation.Y {
        type=MovieSceneScriptingFloatChannel
        default=0
        thinned_from=322
        keys [[0,4.2758],[1,4.6739],[2,4.8582],[3,5.0299],[5,5.3341],[6,5.4697],[8,5.7086],[9,5.8136],[10,5.9387],[11,6.1207],[12,6.3658],[13,6.6611],[14,6.9998],[15,7.375],[17,8.1956],[20,9.4641],[21,9.8626],[22,10.2375],[23,10.5793],[24,10.8764],[25,11.1238],[26,11.3109],[28,11.5795],[30,11.7632],[32,11.8693],[35,11.9149],[55,11.4116],[84,10.0933],[86,10.0542],[87,10.0662],[89,10.143],[93,10.429],[94,10.4842],[95,10.5078],[96,10.4893],[97,10.4149],[98,10.2686],[99,10.0386],[100,9.7365],[101,9.3755],[102,8.9535],[103,8.4776],[104,7.951],[105,7.3739],[106,6.7543],[107,6.1023],[110,4.0918],[111,3.4353],[112,2.7981],[113,2.1855],[114,1.6035],[115,1.0439],[116,0.5026],[117,-0.009],[118,-0.4838],[119,-0.9168],[120,-1.3023],[121,-1.6352],[122,-1.9129],[123,-2.1427],[124,-2.3353],[125,-2.4877],[126,-2.5992],[127,-2.6744],[128,-2.7198],[129,-2.7431],[133,-2.7581],[137,-2.7915],[138,-2.7927],[140,-2.7806],[143,-2.7368],[147,-2.6539],[155,-2.4631],[158,-2.4086],[162,-2.3632],[163,-2.3571],[164,-2.3564],[166,-2.37],[169,-2.4126],[173,-2.4834],[175,-2.4986],[176,-2.5008],[177,-2.4909],[178,-2.4711],[180,-2.4053],[182,-2.3291],[184,-2.2656],[185,-2.2413],[195,-2.1072],[199,-2.0716],[202,-2.0637],[204,-2.0711],[206,-2.0895],[208,-2.1214],[210,-2.1678],[211,-2.1958],[213,-2.2649],[214,-2.3098],[215,-2.3618],[216,-2.4075],[217,-2.4369],[218,-2.4494],[219,-2.4515],[220,-2.4609],[221,-2.49],[222,-2.5486],[223,-2.6462],[224,-2.7765],[225,-2.9295],[227,-3.263],[228,-3.4245],[229,-3.5697],[230,-3.6881],[231,-3.7697],[232,-3.8034],[233,-3.7788],[234,-3.7112],[235,-3.6182],[239,-3.1596],[241,-2.9138],[242,-2.7814],[244,-2.5045],[246,-2.2406],[247,-2.1175],[249,-1.8837],[250,-1.7609],[251,-1.6302],[252,-1.4788],[253,-1.3085],[254,-1.1184],[255,-0.9107],[256,-0.6867],[257,-0.4453],[258,-0.1877],[259,0.0796],[260,0.3615],[261,0.6457],[262,0.9224],[263,1.1821],[264,1.4088],[265,1.6026],[266,1.7605],[267,1.8818],[268,1.9732],[269,2.0249],[273,2.16],[276,2.2804],[277,2.3246],[281,2.4773],[288,2.7895],[289,2.8413],[292,2.9864],[294,3.1013],[297,3.2966],[298,3.3717],[299,3.4573],[301,3.6527],[303,3.8733],[312,4.9627],[315,5.2914],[317,5.4803],[320,5.721],[354,4.2758]]
       }
       chest_ctrl.Rotation.Z {
        type=MovieSceneScriptingFloatChannel
        default=0
        thinned_from=322
        keys [[0,1.1133],[1,1.1119],[2,1.0355],[3,0.9396],[4,0.8261],[5,0.6964],[6,0.5504],[7,0.3893],[8,0.2145],[9,0.025],[10,-0.1737],[11,-0.3815],[12,-0.5958],[13,-0.8147],[14,-1.0407],[16,-1.5039],[17,-1.7429],[18,-1.9856],[19,-2.2239],[21,-2.7089],[22,-2.9451],[24,-3.3931],[25,-3.6009],[26,-3.7942],[27,-3.9714],[28,-4.1247],[29,-4.2535],[30,-4.3577],[31,-4.438],[32,-4.5034],[33,-4.5493],[39,-4.7158],[44,-4.8123],[49,-4.8726],[56,-4.8877],[79,-4.6881],[83,-4.6822],[84,-4.7141],[85,-4.8035],[86,-4.9511],[87,-5.1812],[88,-5.4792],[89,-5.8299],[90,-6.2477],[91,-6.7351],[92,-7.2707],[94,-8.3979],[95,-8.9458],[96,-9.4534],[97,-9.8984],[98,-10.2564],[99,-10.5108],[100,-10.6793],[101,-10.7707],[102,-10.789],[103,-10.7351],[104,-10.6096],[105,-10.4143],[106,-10.1473],[107,-9.831],[109,-9.1288],[112,-8.015],[113,-7.6582],[116,-6.6721],[117,-6.3599],[118,-6.0658],[119,-5.8097],[120,-5.6069],[121,-5.4677],[122,-5.3995],[123,-5.3919],[124,-5.4238],[125,-5.5038],[126,-5.6367],[127,-5.8122],[128,-6.0211],[131,-6.7068],[132,-6.9146],[133,-7.1032],[134,-7.2667],[135,-7.4087],[136,-7.5305],[137,-7.6282],[138,-7.7092],[139,-7.7703],[141,-7.8516],[144,-7.8968],[147,-7.8692],[165,-7.5433],[172,-7.5029],[174,-7.5317],[176,-7.6084],[177,-7.668],[179,-7.857],[181,-8.0783],[183,-8.2648],[184,-8.3404],[185,-8.397],[189,-8.4696],[193,-8.4666],[197,-8.3688],[200,-8.234],[203,-8.029],[206,-7.7541],[208,-7.5255],[210,-7.2562],[212,-6.9432],[213,-6.771],[214,-6.5778],[215,-6.352],[216,-6.0885],[217,-5.7857],[218,-5.4512],[219,-5.0808],[220,-4.6535],[221,-4.1609],[222,-3.5856],[223,-2.932],[224,-2.2225],[225,-1.4776],[226,-0.6992],[227,0.0903],[228,0.884],[229,1.6689],[230,2.4326],[231,3.1547],[232,3.8248],[233,4.4292],[234,4.9583],[235,5.3938],[236,5.7589],[237,6.0712],[238,6.3406],[239,6.5624],[240,6.7423],[241,6.8826],[242,6.9852],[243,7.0517],[244,7.0849],[245,7.0904],[246,7.0676],[247,7.0187],[248,6.9477],[249,6.8557],[250,6.7451],[251,6.6186],[252,6.4779],[255,5.9954],[258,5.486],[260,5.1791],[261,5.0422],[262,4.9226],[263,4.8167],[264,4.7308],[265,4.6629],[266,4.6189],[267,4.5999],[268,4.6035],[269,4.6413],[272,4.7769],[276,4.9087],[292,5.2329],[294,5.2363],[296,5.2054],[297,5.1787],[298,5.1345],[299,5.069],[300,4.9842],[301,4.8843],[302,4.7674],[303,4.6356],[304,4.4929],[306,4.1796],[309,3.6512],[313,2.9244],[315,2.5754],[316,2.4079],[317,2.2478],[318,2.0964],[319,1.9547],[320,1.8235],[354,1.1133]]
       }
       chest_ctrl.Scale.X {type=MovieSceneScriptingFloatChannel, default=1, const [1,322]}
       chest_ctrl.Scale.Y {type=MovieSceneScriptingFloatChannel, default=1, const [1,322]}
       chest_ctrl.Scale.Z {type=MovieSceneScriptingFloatChannel, default=1, const [1,322]}
       chest_tan_ctrl.Location.X {
        type=MovieSceneScriptingFloatChannel
        default=0
        thinned_from=322
        keys [[0,0.0026],[1,-0.0092],[2,-0.0108],[3,-0.0183],[4,-0.0249],[5,-0.0314],[6,-0.0375],[7,-0.0383],[8,-0.0436],[9,-0.0486],[10,-0.0545],[11,-0.0628],[12,-0.0686],[13,-0.08],[14,-0.0893],[15,-0.1],[16,-0.1101],[17,-0.1229],[18,-0.1341],[20,-0.1553],[21,-0.1563],[22,-0.1727],[24,-0.1862],[25,-0.1905],[26,-0.194],[29,-0.2001],[30,-0.209],[31,-0.2011],[32,-0.2015],[33,-0.2011],[34,-0.2017],[36,-0.1996],[37,-0.1999],[41,-0.1938],[42,-0.1919],[43,-0.1912],[45,-0.1877],[46,-0.1875],[47,-0.1858],[49,-0.1847],[50,-0.1811],[52,-0.1797],[54,-0.179],[55,-0.1824],[58,-0.1803],[59,-0.1765],[60,-0.1795],[62,-0.1774],[63,-0.1774],[64,-0.1763],[65,-0.1778],[66,-0.1773],[67,-0.1751],[68,-0.174],[69,-0.1758],[70,-0.1746],[71,-0.1766],[72,-0.1721],[73,-0.1709],[74,-0.1725],[75,-0.1716],[76,-0.1735],[77,-0.1688],[78,-0.1683],[79,-0.1701],[80,-0.1692],[81,-0.1702],[82,-0.1692],[83,-0.1653],[84,-0.1665],[85,-0.166],[86,-0.1687],[87,-0.1672],[88,-0.1677],[89,-0.1694],[90,-0.1729],[91,-0.1722],[92,-0.1744],[93,-0.1773],[94,-0.1786],[95,-0.1786],[96,-0.1774],[97,-0.1777],[98,-0.173],[99,-0.1709],[100,-0.1664],[101,-0.1612],[102,-0.1414],[103,-0.1371],[104,-0.1184],[105,-0.1108],[106,-0.0936],[107,-0.0887],[108,-0.0656],[109,-0.052],[110,-0.0455],[111,-0.0283],[112,-0.0115],[113,0.0016],[114,0.0162],[115,0.03],[116,0.0316],[117,0.0463],[118,0.058],[119,0.0567],[120,0.0723],[121,0.0722],[122,0.0844],[123,0.0833],[124,0.0834],[125,0.0918],[126,0.0917],[127,0.0946],[128,0.092],[130,0.0921],[131,0.0923],[132,0.092],[133,0.0924],[134,0.0924],[135,0.0931],[136,0.0928],[137,0.093],[138,0.0929],[140,0.0923],[141,0.0912],[142,0.0903],[143,0.0897],[145,0.0877],[150,0.084],[151,0.0825],[152,0.0818],[154,0.08],[155,0.0796],[156,0.0782],[157,0.0781],[158,0.0775],[159,0.0771],[161,0.077],[162,0.0766],[163,0.0765],[164,0.0781],[165,0.0782],[166,0.0796],[167,0.0801],[168,0.0791],[169,0.08],[170,0.0816],[173,0.0835],[174,0.0845],[175,0.0861],[176,0.0863],[177,0.0871],[178,0.0868],[179,0.0844],[180,0.0875],[181,0.0866],[182,0.0898],[183,0.0837],[184,0.0867],[185,0.0865],[186,0.0909],[187,0.0911],[188,0.0906],[189,0.0954],[190,0.0902],[191,0.0899],[192,0.0942],[193,0.0954],[194,0.0945],[196,0.0945],[197,0.0939],[199,0.0947],[200,0.0948],[201,0.0963],[202,0.0966],[203,0.0975],[204,0.0992],[205,0.0969],[206,0.0977],[207,0.0999],[208,0.0981],[209,0.1001],[210,0.0999],[211,0.0988],[212,0.0986],[213,0.0989],[214,0.0997],[215,0.0997],[216,0.1007],[217,0.1033],[218,0.1038],[219,0.1068],[220,0.1102],[222,0.1231],[223,0.1315],[224,0.1412],[225,0.1527],[226,0.1637],[227,0.1734],[228,0.184],[229,0.1953],[231,0.2093],[232,0.2147],[233,0.2163],[234,0.2171],[235,0.2151],[236,0.2104],[237,0.2076],[238,0.2009],[239,0.1971],[241,0.1847],[243,0.1696],[244,0.1613],[245,0.1537],[246,0.1451],[247,0.1355],[248,0.1283],[249,0.1188],[250,0.1102],[251,0.1026],[252,0.0937],[253,0.0839],[254,0.0753],[255,0.0684],[256,0.0602],[257,0.0518],[258,0.0423],[259,0.0338],[260,0.0266],[261,0.0166],[262,0.0079],[263,0.0012],[264,-0.0075],[265,-0.0126],[266,-0.0183],[267,-0.0217],[268,-0.0241],[269,-0.0257],[270,-0.0269],[271,-0.0305],[272,-0.0313],[273,-0.031],[274,-0.0322],[275,-0.0329],[276,-0.0325],[277,-0.0372],[278,-0.038],[279,-0.0393],[280,-0.0384],[281,-0.0396],[282,-0.0405],[283,-0.038],[284,-0.0424],[285,-0.0416],[286,-0.0411],[287,-0.0421],[288,-0.0452],[289,-0.044],[290,-0.0424],[291,-0.0467],[293,-0.0435],[294,-0.0449],[295,-0.0468],[296,-0.0455],[297,-0.0479],[298,-0.0469],[299,-0.0505],[300,-0.0493],[301,-0.0523],[302,-0.0497],[304,-0.057],[305,-0.0555],[306,-0.0566],[307,-0.061],[308,-0.0598],[309,-0.0612],[310,-0.0662],[311,-0.0647],[312,-0.0698],[313,-0.0688],[314,-0.0739],[315,-0.0729],[316,-0.0785],[317,-0.0811],[318,-0.0861],[319,-0.0827],[320,-0.0849],[354,0.0026]]
       }
       chest_tan_ctrl.Location.Y {
        type=MovieSceneScriptingFloatChannel
        default=-0.0303
        thinned_from=322
        keys [[0,0.4065],[1,0.4078],[2,0.4078],[3,0.409],[4,0.4139],[6,0.4114],[7,0.4153],[9,0.4101],[11,0.4128],[12,0.4089],[14,0.408],[15,0.4029],[16,0.4021],[17,0.3984],[18,0.3931],[22,0.3771],[23,0.37],[24,0.3695],[25,0.3614],[26,0.3568],[28,0.351],[29,0.3457],[30,0.3426],[31,0.3376],[32,0.3386],[33,0.3325],[35,0.3246],[36,0.33],[37,0.326],[40,0.316],[41,0.3198],[42,0.3172],[44,0.313],[45,0.31],[46,0.3152],[47,0.3132],[48,0.3121],[49,0.3125],[50,0.3149],[51,0.3099],[52,0.311],[53,0.314],[55,0.3176],[56,0.3202],[57,0.3188],[60,0.3285],[63,0.3405],[64,0.3391],[65,0.345],[66,0.3478],[67,0.3498],[68,0.3542],[70,0.3602],[71,0.364],[72,0.3691],[73,0.3692],[75,0.376],[77,0.3814],[78,0.3819],[84,0.393],[86,0.3988],[87,0.3997],[88,0.408],[89,0.4136],[90,0.417],[91,0.4264],[93,0.4378],[96,0.456],[97,0.4655],[98,0.4716],[99,0.4808],[100,0.4832],[101,0.4925],[102,0.5],[103,0.5091],[104,0.5133],[105,0.5219],[107,0.5352],[109,0.5453],[112,0.5561],[113,0.5548],[114,0.5568],[116,0.5572],[117,0.5547],[118,0.5541],[120,0.5408],[122,0.5304],[123,0.5215],[124,0.5148],[126,0.4951],[127,0.484],[128,0.4784],[129,0.4667],[130,0.458],[132,0.4447],[133,0.4399],[134,0.4365],[135,0.4268],[136,0.4223],[137,0.4188],[138,0.4191],[139,0.4247],[140,0.4243],[142,0.4297],[143,0.4342],[144,0.4319],[145,0.4416],[147,0.4514],[148,0.4543],[150,0.4646],[151,0.4674],[152,0.4726],[154,0.4802],[156,0.486],[157,0.4875],[158,0.4939],[159,0.4959],[160,0.4995],[161,0.5009],[162,0.5058],[164,0.504],[166,0.5091],[168,0.5113],[169,0.5068],[170,0.5119],[171,0.5104],[173,0.5144],[174,0.5112],[176,0.5118],[177,0.5082],[178,0.5087],[180,0.5076],[181,0.5011],[183,0.4992],[184,0.4968],[185,0.4965],[187,0.4839],[188,0.4885],[190,0.4769],[191,0.4776],[192,0.4725],[193,0.4699],[194,0.4688],[195,0.4658],[196,0.466],[197,0.4636],[198,0.4546],[199,0.4528],[200,0.4547],[201,0.4534],[202,0.4557],[203,0.4455],[204,0.4451],[205,0.448],[206,0.4486],[207,0.4422],[209,0.4429],[210,0.4395],[211,0.4417],[212,0.4414],[213,0.4448],[215,0.4423],[217,0.4468],[218,0.4443],[219,0.4494],[220,0.4477],[221,0.4495],[223,0.4454],[224,0.4512],[225,0.4499],[226,0.4463],[227,0.456],[228,0.4514],[229,0.4583],[230,0.4528],[231,0.4565],[232,0.4615],[233,0.46],[234,0.46],[235,0.4706],[236,0.4705],[237,0.4785],[238,0.4789],[239,0.4905],[240,0.49],[241,0.5002],[243,0.5105],[244,0.5239],[246,0.5346],[247,0.5481],[248,0.5543],[249,0.5664],[250,0.5711],[252,0.5917],[253,0.5936],[255,0.6091],[256,0.6156],[257,0.6239],[260,0.6408],[262,0.6487],[264,0.6552],[265,0.6564],[266,0.6593],[267,0.6594],[268,0.661],[271,0.662],[273,0.6611],[274,0.6571],[275,0.662],[276,0.657],[277,0.6554],[279,0.6462],[280,0.6469],[282,0.6455],[283,0.6411],[284,0.6408],[285,0.6465],[286,0.6425],[288,0.6428],[290,0.6417],[294,0.6447],[296,0.6499],[297,0.6507],[298,0.6542],[300,0.6588],[301,0.6627],[304,0.6707],[305,0.6719],[306,0.6768],[307,0.6786],[308,0.6837],[309,0.6852],[310,0.6889],[311,0.69],[312,0.6945],[313,0.6953],[314,0.6994],[316,0.7002],[317,0.7044],[319,0.7051],[320,0.7082],[354,0.4065]]
       }
       chest_tan_ctrl.Location.Z {
        type=MovieSceneScriptingFloatChannel
        default=0.265
        thinned_from=322
        keys [[0,0.246],[1,0.2466],[3,0.2465],[4,0.2487],[7,0.2491],[8,0.2473],[10,0.2478],[11,0.2475],[12,0.2462],[14,0.2443],[16,0.2437],[17,0.2427],[18,0.2439],[19,0.2406],[20,0.2345],[21,0.2386],[22,0.2365],[23,0.2383],[24,0.2351],[25,0.2366],[26,0.2327],[27,0.2334],[28,0.2258],[29,0.2295],[30,0.2307],[31,0.2261],[32,0.2254],[34,0.226],[36,0.2231],[37,0.2241],[38,0.223],[40,0.2222],[41,0.221],[42,0.2222],[47,0.2212],[48,0.2215],[49,0.2205],[52,0.2203],[53,0.2269],[54,0.2238],[58,0.2238],[59,0.2273],[62,0.2284],[63,0.2315],[65,0.232],[66,0.2354],[67,0.2323],[69,0.2391],[70,0.2391],[71,0.2396],[73,0.2395],[74,0.2429],[75,0.2429],[76,0.2463],[77,0.2428],[78,0.2431],[79,0.2464],[82,0.2464],[83,0.2484],[84,0.2474],[85,0.2492],[86,0.2486],[88,0.2518],[89,0.2546],[90,0.2566],[91,0.2597],[92,0.2621],[93,0.2673],[95,0.2715],[96,0.2746],[97,0.2793],[98,0.2798],[99,0.2849],[100,0.2873],[101,0.2929],[102,0.2937],[103,0.2961],[104,0.3006],[105,0.3029],[106,0.307],[108,0.3116],[109,0.3162],[110,0.3144],[111,0.3154],[112,0.3176],[113,0.319],[114,0.3181],[115,0.3186],[116,0.3178],[117,0.3208],[118,0.3159],[119,0.3125],[120,0.3113],[121,0.3055],[122,0.3022],[123,0.3002],[124,0.2937],[126,0.2852],[127,0.2794],[128,0.2766],[130,0.2657],[133,0.255],[134,0.2519],[135,0.2502],[137,0.2442],[138,0.2458],[140,0.2455],[141,0.2447],[142,0.2457],[143,0.2476],[144,0.247],[145,0.2499],[146,0.2518],[149,0.2553],[150,0.2573],[151,0.2587],[152,0.2606],[153,0.2619],[154,0.2638],[155,0.2649],[156,0.2666],[157,0.2662],[159,0.27],[161,0.2721],[162,0.2721],[163,0.2733],[165,0.2737],[166,0.2749],[167,0.2749],[168,0.2737],[170,0.2739],[171,0.2731],[172,0.2734],[174,0.2722],[175,0.2707],[176,0.2706],[178,0.2682],[179,0.2655],[180,0.2641],[181,0.2666],[182,0.2638],[183,0.2628],[184,0.2594],[185,0.2584],[186,0.2592],[187,0.2581],[188,0.2555],[189,0.2545],[190,0.2522],[191,0.2551],[192,0.2543],[193,0.2523],[194,0.2525],[195,0.2509],[196,0.2505],[197,0.2513],[198,0.2506],[199,0.2514],[200,0.249],[201,0.2492],[202,0.25],[203,0.2498],[205,0.2503],[206,0.2513],[207,0.2509],[208,0.2491],[209,0.2504],[214,0.2523],[217,0.2545],[218,0.2546],[220,0.2559],[222,0.2559],[224,0.2594],[225,0.2621],[226,0.264],[227,0.2645],[228,0.2672],[229,0.2732],[230,0.2724],[231,0.2766],[232,0.2792],[233,0.2795],[234,0.2848],[235,0.285],[236,0.2881],[238,0.2924],[239,0.2972],[241,0.3012],[242,0.3096],[243,0.3084],[244,0.3177],[245,0.3167],[246,0.3255],[247,0.3253],[248,0.3261],[249,0.3343],[250,0.3335],[251,0.343],[252,0.3427],[253,0.3513],[254,0.3506],[255,0.3595],[256,0.3586],[258,0.3657],[260,0.3696],[261,0.3726],[262,0.374],[263,0.3767],[264,0.3774],[265,0.3809],[266,0.381],[268,0.3836],[271,0.3852],[272,0.3852],[273,0.3861],[274,0.385],[277,0.3856],[278,0.3841],[279,0.381],[280,0.3824],[281,0.3825],[283,0.3812],[284,0.3816],[286,0.3804],[287,0.3823],[288,0.3826],[289,0.3815],[292,0.3833],[294,0.3853],[295,0.3855],[296,0.3869],[297,0.3875],[300,0.3912],[302,0.3948],[303,0.3952],[304,0.397],[305,0.397],[307,0.4022],[308,0.4027],[309,0.4045],[310,0.4049],[311,0.4071],[312,0.4075],[313,0.4098],[315,0.4099],[316,0.4128],[318,0.4129],[319,0.4154],[320,0.4154],[354,0.246]]
       }
       chest_tan_ctrl.Rotation.X {
        type=MovieSceneScriptingFloatChannel
        default=0
        thinned_from=322
        keys [[0,1.4612],[1,1.5405],[2,1.5575],[3,1.5668],[4,1.5712],[5,1.5693],[6,1.5614],[7,1.5482],[8,1.5302],[9,1.5057],[10,1.4697],[11,1.4215],[12,1.3599],[13,1.2804],[14,1.1838],[15,1.0681],[16,0.936],[17,0.7893],[18,0.6279],[19,0.4603],[20,0.2848],[21,0.1081],[22,0],[66,0],[67,0.0103],[68,0.09],[69,0.1664],[70,0.2409],[72,0.3749],[73,0.438],[74,0.494],[75,0.5459],[76,0.5928],[77,0.6359],[78,0.6754],[79,0.7114],[81,0.7785],[83,0.8439],[84,0.8809],[85,0.9243],[86,0.973],[87,1.039],[88,1.1384],[89,1.2667],[90,1.4178],[91,1.5882],[92,1.7757],[93,1.9777],[94,2.1917],[97,2.8674],[98,3.0983],[101,3.8221],[102,4.0578],[103,4.2841],[104,4.494],[105,4.6816],[106,4.8414],[107,4.9688],[108,5.0599],[109,5.1112],[110,5.1316],[111,5.1329],[112,5.1169],[113,5.0826],[114,5.0302],[115,4.9606],[116,4.8564],[117,4.7065],[118,4.5167],[119,4.2915],[120,4.0409],[121,3.7667],[122,3.4784],[125,2.5802],[126,2.2882],[127,2.0083],[128,1.743],[129,1.4981],[130,1.2758],[131,1.0754],[132,0.8998],[133,0.7452],[134,0.6164],[135,0.5099],[136,0.4259],[137,0.3672],[138,0.3322],[139,0.3217],[140,0.3301],[141,0.3489],[142,0.3779],[143,0.4141],[144,0.459],[145,0.5116],[146,0.5723],[147,0.6369],[149,0.7813],[152,1.0084],[154,1.1511],[155,1.2199],[157,1.3375],[158,1.3908],[159,1.4398],[161,1.5248],[162,1.5649],[163,1.5956],[164,1.6168],[166,1.6374],[167,1.6343],[169,1.6144],[170,1.5906],[171,1.5621],[172,1.526],[174,1.4375],[176,1.3319],[180,1.1072],[182,1.0012],[186,0.7973],[188,0.705],[189,0.6631],[190,0.6258],[191,0.5928],[192,0.5641],[193,0.5394],[194,0.5218],[195,0.5065],[196,0.4954],[197,0.4856],[198,0.4791],[199,0.4747],[200,0.4717],[201,0.4706],[202,0.472],[203,0.4725],[204,0.476],[206,0.4859],[208,0.4985],[209,0.5074],[210,0.5194],[211,0.5341],[212,0.5523],[213,0.573],[215,0.6222],[216,0.6424],[217,0.6542],[218,0.659],[219,0.654],[220,0.6359],[221,0.6038],[222,0.5597],[223,0.5267],[224,0.5006],[225,0.4834],[226,0.4759],[227,0.483],[228,0.5072],[229,0.5576],[230,0.6325],[231,0.7324],[232,0.8586],[233,1.0121],[234,1.193],[235,1.4005],[236,1.6263],[237,1.8681],[238,2.1227],[239,2.3912],[240,2.6703],[241,2.9601],[242,3.2622],[248,5.1593],[249,5.4683],[250,5.7618],[251,6.0433],[252,6.3096],[253,6.562],[254,6.7988],[256,7.23],[258,7.6046],[260,7.9422],[262,8.2435],[264,8.5078],[266,8.7305],[268,8.9107],[270,9.0488],[272,9.1475],[274,9.2092],[283,9.2403],[291,9.392],[299,9.7364],[312,10.4372],[320,10.6976],[354,1.4612]]
       }
       chest_tan_ctrl.Rotation.Y {
        type=MovieSceneScriptingFloatChannel
        default=0
        thinned_from=322
        keys [[0,2.5399],[1,2.8933],[2,3.0666],[3,3.2303],[4,3.3836],[5,3.5288],[6,3.665],[8,3.9127],[9,4.0254],[10,4.1567],[11,4.3403],[12,4.5818],[13,4.8697],[14,5.1995],[15,5.5637],[16,5.9553],[17,6.3612],[20,7.5999],[21,7.9922],[22,8.3621],[23,8.7009],[24,8.9977],[25,9.2461],[26,9.4372],[28,9.7153],[29,9.823],[31,9.9751],[33,10.0512],[36,10.069],[55,9.5367],[83,8.2657],[85,8.206],[86,8.201],[87,8.2187],[89,8.304],[93,8.5968],[94,8.6522],[95,8.6749],[96,8.6551],[97,8.5806],[98,8.4352],[99,8.2085],[100,7.9121],[101,7.5589],[102,7.148],[103,6.6874],[104,6.1812],[105,5.631],[106,5.0452],[110,2.5856],[111,1.9907],[112,1.4171],[113,0.8691],[114,0.3525],[115,0],[264,0],[265,0.1313],[266,0.2755],[267,0.3861],[268,0.47],[269,0.5179],[271,0.582],[272,0.612],[273,0.6447],[275,0.7221],[276,0.7586],[277,0.8004],[278,0.8362],[279,0.8739],[281,0.9441],[282,0.9845],[283,1.0218],[284,1.0629],[285,1.1011],[287,1.183],[288,1.2219],[289,1.2666],[290,1.3054],[292,1.3888],[294,1.4827],[297,1.6422],[298,1.7027],[299,1.7733],[300,1.8521],[301,1.9355],[303,2.1177],[312,3.0256],[314,3.2165],[316,3.3931],[318,3.5484],[320,3.6877],[354,2.5399]]
       }
       chest_tan_ctrl.Rotation.Z {
        type=MovieSceneScriptingFloatChannel
        default=0
        thinned_from=322
        keys [[0,0.7914],[1,0.7958],[2,0.7533],[3,0.6951],[4,0.6275],[5,0.5479],[6,0.4602],[7,0.3615],[8,0.2531],[9,0.1386],[10,0.0153],[11,0],[226,0],[227,0.0506],[228,0.5542],[229,1.0531],[230,1.5395],[231,1.9975],[232,2.4226],[233,2.8058],[234,3.1404],[235,3.4152],[236,3.6447],[237,3.8416],[238,4.0116],[239,4.1514],[240,4.2638],[241,4.3511],[242,4.4136],[243,4.4531],[244,4.472],[245,4.4722],[246,4.4551],[247,4.4209],[248,4.3726],[249,4.3115],[250,4.2381],[251,4.1561],[253,3.9635],[258,3.4165],[259,3.3114],[261,3.1224],[262,3.0435],[263,2.9709],[264,2.9147],[265,2.8667],[266,2.836],[267,2.8216],[268,2.821],[269,2.8424],[272,2.9221],[276,3.0027],[292,3.2006],[294,3.2002],[296,3.1785],[297,3.1602],[298,3.1322],[299,3.0882],[300,3.0353],[301,2.9701],[302,2.8955],[303,2.8115],[304,2.7218],[306,2.5219],[309,2.1876],[314,1.6164],[315,1.5079],[316,1.4025],[317,1.3008],[318,1.206],[319,1.1168],[320,1.0344],[354,0.7914]]
       }
       chest_tan_ctrl.Scale.X {type=MovieSceneScriptingFloatChannel, default=1, const [1,322]}
       chest_tan_ctrl.Scale.Y {type=MovieSceneScriptingFloatChannel, default=1, const [1,322]}
       chest_tan_ctrl.Scale.Z {type=MovieSceneScriptingFloatChannel, default=1, const [1,322]}
       Weight {type=MovieSceneScriptingFloatChannel, default=1}
      ]
     }
     {
      type=MovieSceneControlRigParameterSection
      channels [
       upperarm_l_fk_ctrl.Location.X {type=MovieSceneScriptingFloatChannel, default=0, thinned_from=10, keys [[9,0],[354,0]]}
       upperarm_l_fk_ctrl.Location.Y {type=MovieSceneScriptingFloatChannel, default=0, thinned_from=10, keys [[9,0],[354,0]]}
       upperarm_l_fk_ctrl.Location.Z {type=MovieSceneScriptingFloatChannel, default=0, thinned_from=10, keys [[9,0],[354,0]]}
       upperarm_l_fk_ctrl.Rotation.X {
        type=MovieSceneScriptingFloatChannel
        default=0
        keys [[1,-1.1795],[49,0.6201],[107,0.0668],[144,6.3752],[192,10.1494],[206,10.5297],[262,-4.3545],[293,-9.0797],[295,-9.1298],[354,-1.1795]]
       }
       upperarm_l_fk_ctrl.Rotation.Y {
        type=MovieSceneScriptingFloatChannel
        default=0
        keys [[9,4.3979],[86,4.0948],[149,4.6412],[200,6.6762],[205,6.6763],[211,6.4733],[226,5.1763],[265,1.7683],[289,0.1346],[295,0.1654],[354,4.3979]]
       }
       upperarm_l_fk_ctrl.Rotation.Z {
        type=MovieSceneScriptingFloatChannel
        default=0
        keys [[9,0.5556],[89,1.267],[135,8.5887],[205,8.6925],[242,8.712],[265,6.8419],[295,7.0166],[354,0.5556]]
       }
       upperarm_l_fk_ctrl.Scale.X {type=MovieSceneScriptingFloatChannel, default=0, const [0,10]}
       upperarm_l_fk_ctrl.Scale.Y {type=MovieSceneScriptingFloatChannel, default=0, const [0,10]}
       upperarm_l_fk_ctrl.Scale.Z {type=MovieSceneScriptingFloatChannel, default=0, const [0,10]}
       lowerarm_l_fk_ctrl.Location.X {type=MovieSceneScriptingFloatChannel, default=0, thinned_from=4, keys [[187,0],[275,0]]}
       lowerarm_l_fk_ctrl.Location.Y {type=MovieSceneScriptingFloatChannel, default=0, thinned_from=4, keys [[187,0],[275,0]]}
       lowerarm_l_fk_ctrl.Location.Z {type=MovieSceneScriptingFloatChannel, default=0, thinned_from=4, keys [[187,0],[275,0]]}
       lowerarm_l_fk_ctrl.Rotation.X {type=MovieSceneScriptingFloatChannel, default=0, const [0,4]}
       lowerarm_l_fk_ctrl.Rotation.Y {type=MovieSceneScriptingFloatChannel, default=0, const [0,4]}
       lowerarm_l_fk_ctrl.Rotation.Z {
        type=MovieSceneScriptingFloatChannel
        default=0
        keys [[101,-4.6916],[130,-7.9949],[187,-9.5422],[208,-9.577],[275,-4.6916],[292,-4.0043],[353,-4.6916]]
       }
       lowerarm_l_fk_ctrl.Scale.X {type=MovieSceneScriptingFloatChannel, default=0, const [0,4]}
       lowerarm_l_fk_ctrl.Scale.Y {type=MovieSceneScriptingFloatChannel, default=0, const [0,4]}
       lowerarm_l_fk_ctrl.Scale.Z {type=MovieSceneScriptingFloatChannel, default=0, const [0,4]}
       upperarm_r_fk_ctrl.Location.X {type=MovieSceneScriptingFloatChannel, default=0, thinned_from=6, keys [[9,0],[354,0]]}
       upperarm_r_fk_ctrl.Location.Y {type=MovieSceneScriptingFloatChannel, default=0, thinned_from=6, keys [[9,0],[354,0]]}
       upperarm_r_fk_ctrl.Location.Z {type=MovieSceneScriptingFloatChannel, default=0, thinned_from=6, keys [[9,0],[354,0]]}
       upperarm_r_fk_ctrl.Rotation.X {type=MovieSceneScriptingFloatChannel, default=0, keys [[9,-1.5345],[86,-1.5705],[147,-14.2243],[212,-14.262],[264,-12.9645],[354,-1.5345]]}
       upperarm_r_fk_ctrl.Rotation.Y {type=MovieSceneScriptingFloatChannel, default=0, keys [[9,1.3386],[86,1.3348],[147,0.004],[212,0],[275,0.1841],[354,1.3386]]}
       upperarm_r_fk_ctrl.Rotation.Z {type=MovieSceneScriptingFloatChannel, default=0, keys [[9,2.2938],[86,2.2874],[112,4.7679],[176,4.771],[275,2.2143],[354,2.2938]]}
       upperarm_r_fk_ctrl.Scale.X {type=MovieSceneScriptingFloatChannel, default=0, const [0,6]}
       upperarm_r_fk_ctrl.Scale.Y {type=MovieSceneScriptingFloatChannel, default=0, const [0,6]}
       upperarm_r_fk_ctrl.Scale.Z {type=MovieSceneScriptingFloatChannel, default=0, const [0,6]}
       Weight {type=MovieSceneScriptingFloatChannel, default=1}
      ]
     }
     {
      type=MovieSceneControlRigParameterSection
      channels [
       thigh_l_fk_ctrl.Location.X {type=MovieSceneScriptingFloatChannel, default=0, const [0,3]}
       thigh_l_fk_ctrl.Location.Y {type=MovieSceneScriptingFloatChannel, default=0, const [0,3]}
       thigh_l_fk_ctrl.Location.Z {type=MovieSceneScriptingFloatChannel, default=0, const [0,3]}
       thigh_l_fk_ctrl.Rotation.X {
        type=MovieSceneScriptingFloatChannel
        default=-0.604
        thinned_from=7
        keys [[4,17.6655],[81,17.6655],[96,17.3532],[204,16.2448],[251,16.5858],[354,17.6655]]
       }
       thigh_l_fk_ctrl.Rotation.Y {
        type=MovieSceneScriptingFloatChannel
        default=-3.8454
        keys [[4,-2.644],[38,-2.8889],[65,-4.0423],[99,-4.2835],[141,-2.4457],[204,-3.6639],[251,1.0661],[298,2.4371],[354,-2.644]]
       }
       thigh_l_fk_ctrl.Rotation.Z {
        type=MovieSceneScriptingFloatChannel
        default=11.8823
        keys [[4,10.3434],[16,10.3434],[57,11.0947],[113,15.6424],[211,22.0584],[236,20.7298],[267,19.802],[318,17.027],[354,10.3434]]
       }
       thigh_l_fk_ctrl.Scale.X {type=MovieSceneScriptingFloatChannel, default=0, const [0,3]}
       thigh_l_fk_ctrl.Scale.Y {type=MovieSceneScriptingFloatChannel, default=0, const [0,3]}
       thigh_l_fk_ctrl.Scale.Z {type=MovieSceneScriptingFloatChannel, default=0, const [0,3]}
       calf_l_fk_ctrl.Location.X {type=MovieSceneScriptingFloatChannel, default=0, const [0,1]}
       calf_l_fk_ctrl.Location.Y {type=MovieSceneScriptingFloatChannel, default=0, const [0,1]}
       calf_l_fk_ctrl.Location.Z {type=MovieSceneScriptingFloatChannel, default=0, const [0,1]}
       calf_l_fk_ctrl.Rotation.X {type=MovieSceneScriptingFloatChannel, default=-0.0032, keys [[16,-0.0009],[212,0.0063],[354,-0.0009]]}
       calf_l_fk_ctrl.Rotation.Y {type=MovieSceneScriptingFloatChannel, default=-0.0112, keys [[16,-0.0098],[212,-0.0232],[354,-0.0098]]}
       calf_l_fk_ctrl.Rotation.Z {
        type=MovieSceneScriptingFloatChannel
        default=12.4724
        keys [[16,-5.9744],[143,-11.2923],[189,-12.337],[285,-10.6436],[317,-8.3428],[354,-5.9744]]
       }
       calf_l_fk_ctrl.Scale.X {type=MovieSceneScriptingFloatChannel, default=0, const [0,1]}
       calf_l_fk_ctrl.Scale.Y {type=MovieSceneScriptingFloatChannel, default=0, const [0,1]}
       calf_l_fk_ctrl.Scale.Z {type=MovieSceneScriptingFloatChannel, default=0, const [0,1]}
       foot_l_fk_ctrl.Location.X {type=MovieSceneScriptingFloatChannel, default=0, const [0,1]}
       foot_l_fk_ctrl.Location.Y {type=MovieSceneScriptingFloatChannel, default=0, const [0,1]}
       foot_l_fk_ctrl.Location.Z {type=MovieSceneScriptingFloatChannel, default=0, const [0,1]}
       foot_l_fk_ctrl.Rotation.X {type=MovieSceneScriptingFloatChannel, default=-0.6607, const [5.5241,1]}
       foot_l_fk_ctrl.Rotation.Y {type=MovieSceneScriptingFloatChannel, default=1.5192, const [0.2547,1]}
       foot_l_fk_ctrl.Rotation.Z {type=MovieSceneScriptingFloatChannel, default=12.3974, const [15.509,1]}
       foot_l_fk_ctrl.Scale.X {type=MovieSceneScriptingFloatChannel, default=0, const [0,1]}
       foot_l_fk_ctrl.Scale.Y {type=MovieSceneScriptingFloatChannel, default=0, const [0,1]}
       foot_l_fk_ctrl.Scale.Z {type=MovieSceneScriptingFloatChannel, default=0, const [0,1]}
       foot_l_ik_ctrl.Location.X {type=MovieSceneScriptingFloatChannel, default=0, thinned_from=6, keys [[0,0],[237,0],[289,1.1625],[320,1.1625],[351,0]]}
       foot_l_ik_ctrl.Location.Y {type=MovieSceneScriptingFloatChannel, default=0, thinned_from=6, keys [[0,0],[237,0],[289,-3.6377],[320,-3.6377],[351,0]]}
       foot_l_ik_ctrl.Location.Z {
        type=MovieSceneScriptingFloatChannel
        default=0
        keys [[0,0.2297],[88,-0.1722],[107,-0.2334],[213,-1.2537],[273,-1.5872],[320,-1.3116],[351,0.2297]]
       }
       foot_l_ik_ctrl.Rotation.X {type=MovieSceneScriptingFloatChannel, default=0, keys [[0,0.5607],[88,1.0544],[211,2.2692],[289,2.6337],[320,2.6337],[351,0.5607]]}
       foot_l_ik_ctrl.Rotation.Y {type=MovieSceneScriptingFloatChannel, default=0, keys [[0,5.9543],[88,7.7486],[213,11.2729],[289,11.6982],[320,11.6982],[351,5.9543]]}
       foot_l_ik_ctrl.Rotation.Z {type=MovieSceneScriptingFloatChannel, default=0, keys [[0,-16.444],[88,-15.3079],[212,-13.1429],[274,-19.8236],[320,-19.7835],[351,-16.444]]}
       foot_l_ik_ctrl.Scale.X {type=MovieSceneScriptingFloatChannel, default=0, const [0,6]}
       foot_l_ik_ctrl.Scale.Y {type=MovieSceneScriptingFloatChannel, default=0, const [0,6]}
       foot_l_ik_ctrl.Scale.Z {type=MovieSceneScriptingFloatChannel, default=0, const [0,6]}
       leg_l_softness {type=MovieSceneScriptingFloatChannel, default=0, const [0,6]}
       thigh_r_fk_ctrl.Location.X {type=MovieSceneScriptingFloatChannel, default=0, const [0,3]}
       thigh_r_fk_ctrl.Location.Y {type=MovieSceneScriptingFloatChannel, default=0, const [0,3]}
       thigh_r_fk_ctrl.Location.Z {type=MovieSceneScriptingFloatChannel, default=0, const [0,3]}
       thigh_r_fk_ctrl.Rotation.X {type=MovieSceneScriptingFloatChannel, default=-1.199, thinned_from=6, keys [[1,5.3883],[81,5.3883],[96,4.8348],[207,5.3504],[352,5.3883]]}
       thigh_r_fk_ctrl.Rotation.Y {
        type=MovieSceneScriptingFloatChannel
        default=0.4728
        keys [[1,2.3035],[16,2.0372],[81,1.7111],[129,1.6557],[207,1.6667],[250,1.6655],[352,2.3035]]
       }
       thigh_r_fk_ctrl.Rotation.Z {
        type=MovieSceneScriptingFloatChannel
        default=12.5207
        keys [[1,25.2548],[16,25.2548],[52,28.273],[85,31.4756],[207,33.7052],[271,35.7001],[352,25.2548]]
       }
       thigh_r_fk_ctrl.Scale.X {type=MovieSceneScriptingFloatChannel, default=0, const [0,3]}
       thigh_r_fk_ctrl.Scale.Y {type=MovieSceneScriptingFloatChannel, default=0, const [0,3]}
       thigh_r_fk_ctrl.Scale.Z {type=MovieSceneScriptingFloatChannel, default=0, const [0,3]}
       calf_r_fk_ctrl.Location.X {type=MovieSceneScriptingFloatChannel, default=0, thinned_from=4, keys [[105,0],[339,0]]}
       calf_r_fk_ctrl.Location.Y {type=MovieSceneScriptingFloatChannel, default=0, thinned_from=4, keys [[105,0],[339,0]]}
       calf_r_fk_ctrl.Location.Z {type=MovieSceneScriptingFloatChannel, default=0, thinned_from=4, keys [[105,0],[339,0]]}
       calf_r_fk_ctrl.Rotation.X {type=MovieSceneScriptingFloatChannel, default=-0.0264, thinned_from=5, keys [[16,-0.0615],[207,-0.0615],[284,-0.1107],[339,-0.0615]]}
       calf_r_fk_ctrl.Rotation.Y {type=MovieSceneScriptingFloatChannel, default=-0.201, thinned_from=5, keys [[16,0.3615],[207,0.3615],[284,0.4731],[339,0.3615]]}
       calf_r_fk_ctrl.Rotation.Z {type=MovieSceneScriptingFloatChannel, default=12.4708, const [-22.2781,4]}
       calf_r_fk_ctrl.Scale.X {type=MovieSceneScriptingFloatChannel, default=0, const [0,4]}
       calf_r_fk_ctrl.Scale.Y {type=MovieSceneScriptingFloatChannel, default=0, const [0,4]}
       calf_r_fk_ctrl.Scale.Z {type=MovieSceneScriptingFloatChannel, default=0, const [0,4]}
       foot_r_fk_ctrl.Location.X {type=MovieSceneScriptingFloatChannel, default=0, const [0,2]}
       foot_r_fk_ctrl.Location.Y {type=MovieSceneScriptingFloatChannel, default=0, const [0,2]}
       foot_r_fk_ctrl.Location.Z {type=MovieSceneScriptingFloatChannel, default=0, const [0,2]}
       foot_r_fk_ctrl.Rotation.X {type=MovieSceneScriptingFloatChannel, default=-0.4513, keys [[0,16.9708],[251,9.3981],[351,16.9708]]}
       foot_r_fk_ctrl.Rotation.Y {type=MovieSceneScriptingFloatChannel, default=1.4464, const [1.4464,2]}
       foot_r_fk_ctrl.Rotation.Z {type=MovieSceneScriptingFloatChannel, default=12.3967, const [25.9607,2]}
       foot_r_fk_ctrl.Scale.X {type=MovieSceneScriptingFloatChannel, default=0, const [0,2]}
       foot_r_fk_ctrl.Scale.Y {type=MovieSceneScriptingFloatChannel, default=0, const [0,2]}
       foot_r_fk_ctrl.Scale.Z {type=MovieSceneScriptingFloatChannel, default=0, const [0,2]}
       foot_r_ik_ctrl.Location.X {
        type=MovieSceneScriptingFloatChannel
        default=0
        keys [[0,0],[19,-2.1432],[64,-4.0205],[81,-3.9116],[170,-1.3872],[237,0],[289,-4.0441],[351,0]]
       }
       foot_r_ik_ctrl.Location.Y {
        type=MovieSceneScriptingFloatChannel
        default=0
        keys [[0,0],[19,-1.3918],[64,-2.6109],[81,-2.5402],[170,-0.9008],[237,0],[289,-7.7169],[351,0]]
       }
       foot_r_ik_ctrl.Location.Z {
        type=MovieSceneScriptingFloatChannel
        default=-0.2095
        keys [[0,-3.0981],[19,-3.4943],[64,-4.0057],[81,-3.665],[107,-0.2095],[170,-0.0691],[237,0],[289,-4.4474],[351,-3.0981]]
       }
       foot_r_ik_ctrl.Rotation.X {
        type=MovieSceneScriptingFloatChannel
        default=0
        thinned_from=8
        keys [[0,-0.5025],[19,0],[81,0],[170,16.8414],[237,0],[289,0.054],[351,-0.5025]]
       }
       foot_r_ik_ctrl.Rotation.Y {
        type=MovieSceneScriptingFloatChannel
        default=0
        thinned_from=8
        keys [[0,10.7955],[22,9.5688],[65,1.204],[81,0],[237,0],[289,11.1793],[351,10.7955]]
       }
       foot_r_ik_ctrl.Rotation.Z {
        type=MovieSceneScriptingFloatChannel
        default=0
        thinned_from=8
        keys [[0,0.5137],[19,0],[64,-2.7368],[237,-2.7368],[289,-6.9142],[351,0.5137]]
       }
       foot_r_ik_ctrl.Scale.X {type=MovieSceneScriptingFloatChannel, default=0, const [0,8]}
       foot_r_ik_ctrl.Scale.Y {type=MovieSceneScriptingFloatChannel, default=0, const [0,8]}
       foot_r_ik_ctrl.Scale.Z {type=MovieSceneScriptingFloatChannel, default=0, const [0,8]}
       leg_r_softness {type=MovieSceneScriptingFloatChannel, default=0, const [0,8]}
       leg_r_pv_ik_ctrl.Location.X {type=MovieSceneScriptingFloatChannel, default=0, keys [[237,0],[289,11.0972]]}
       leg_r_pv_ik_ctrl.Location.Y {type=MovieSceneScriptingFloatChannel, default=0, keys [[237,0],[289,-5.3138]]}
       leg_r_pv_ik_ctrl.Location.Z {type=MovieSceneScriptingFloatChannel, default=0, keys [[237,0],[289,0.7038]]}
       leg_r_pv_ik_ctrl.Rotation.X {type=MovieSceneScriptingFloatChannel, default=0, const [0,2]}
       leg_r_pv_ik_ctrl.Rotation.Y {type=MovieSceneScriptingFloatChannel, default=0, const [0,2]}
       leg_r_pv_ik_ctrl.Rotation.Z {type=MovieSceneScriptingFloatChannel, default=0, const [0,2]}
       leg_r_pv_ik_ctrl.Scale.X {type=MovieSceneScriptingFloatChannel, default=0, const [0,2]}
       leg_r_pv_ik_ctrl.Scale.Y {type=MovieSceneScriptingFloatChannel, default=0, const [0,2]}
       leg_r_pv_ik_ctrl.Scale.Z {type=MovieSceneScriptingFloatChannel, default=0, const [0,2]}
       body_ctrl.Location.X {type=MovieSceneScriptingFloatChannel, default=0, const [0,4]}
       body_ctrl.Location.Y {type=MovieSceneScriptingFloatChannel, default=0, const [0,4]}
       body_ctrl.Location.Z {type=MovieSceneScriptingFloatChannel, default=0, const [0,4]}
       body_ctrl.Rotation.X {type=MovieSceneScriptingFloatChannel, default=0, keys [[0,-10.7139],[237,-14.0066],[319,-5.419],[351,-10.7139]]}
       body_ctrl.Rotation.Y {type=MovieSceneScriptingFloatChannel, default=0, keys [[0,2.2727],[237,-6.6866],[319,1.1143],[351,2.2727]]}
       body_ctrl.Rotation.Z {type=MovieSceneScriptingFloatChannel, default=0, keys [[0,-0.3821],[237,-0.193],[319,0.1698],[351,-0.3821]]}
       body_ctrl.Scale.X {type=MovieSceneScriptingFloatChannel, default=0, const [0,4]}
       body_ctrl.Scale.Y {type=MovieSceneScriptingFloatChannel, default=0, const [0,4]}
       body_ctrl.Scale.Z {type=MovieSceneScriptingFloatChannel, default=0, const [0,4]}
       hips_ctrl.Location.X {type=MovieSceneScriptingFloatChannel, default=0, const [0,2]}
       hips_ctrl.Location.Y {type=MovieSceneScriptingFloatChannel, default=0, const [0,2]}
       hips_ctrl.Location.Z {type=MovieSceneScriptingFloatChannel, default=0, const [0,2]}
       hips_ctrl.Rotation.X {type=MovieSceneScriptingFloatChannel, default=0, keys [[50,0],[81,0.0033],[150,0]]}
       hips_ctrl.Rotation.Y {type=MovieSceneScriptingFloatChannel, default=0, keys [[50,0],[81,2.252],[150,0]]}
       hips_ctrl.Rotation.Z {type=MovieSceneScriptingFloatChannel, default=0, keys [[50,0],[81,0.1491],[150,0]]}
       hips_ctrl.Scale.X {type=MovieSceneScriptingFloatChannel, default=0, const [0,2]}
       hips_ctrl.Scale.Y {type=MovieSceneScriptingFloatChannel, default=0, const [0,2]}
       hips_ctrl.Scale.Z {type=MovieSceneScriptingFloatChannel, default=0, const [0,2]}
       spine_01_ctrl.Location.X {type=MovieSceneScriptingFloatChannel, default=0, const [0,6]}
       spine_01_ctrl.Location.Y {type=MovieSceneScriptingFloatChannel, default=0, const [0,6]}
       spine_01_ctrl.Location.Z {type=MovieSceneScriptingFloatChannel, default=0, const [0,6]}
       spine_01_ctrl.Rotation.X {
        type=MovieSceneScriptingFloatChannel
        default=0
        keys [[9,-2.7672],[90,-2.7672],[150,-4.3798],[214,-3.0323],[289,-8.328],[346,-2.7672],[347,-2.7672]]
       }
       spine_01_ctrl.Rotation.Y {type=MovieSceneScriptingFloatChannel, default=0, keys [[9,0],[90,-1.1939],[150,6.9415],[214,10.0712],[289,0],[347,0]]}
       spine_01_ctrl.Rotation.Z {type=MovieSceneScriptingFloatChannel, default=0, keys [[9,0],[90,0.234],[150,2.1787],[214,1.0313],[289,0],[347,0]]}
       spine_01_ctrl.Scale.X {type=MovieSceneScriptingFloatChannel, default=0, const [0,6]}
       spine_01_ctrl.Scale.Y {type=MovieSceneScriptingFloatChannel, default=0, const [0,6]}
       spine_01_ctrl.Scale.Z {type=MovieSceneScriptingFloatChannel, default=0, const [0,6]}
       spine_03_ctrl.Location.X {type=MovieSceneScriptingFloatChannel, default=0, const [0,5]}
       spine_03_ctrl.Location.Y {type=MovieSceneScriptingFloatChannel, default=0, const [0,5]}
       spine_03_ctrl.Location.Z {type=MovieSceneScriptingFloatChannel, default=0, const [0,5]}
       spine_03_ctrl.Rotation.X {type=MovieSceneScriptingFloatChannel, default=0, thinned_from=5, keys [[1,0],[237,0],[313,-1.6407],[351,0]]}
       spine_03_ctrl.Rotation.Y {type=MovieSceneScriptingFloatChannel, default=0, thinned_from=5, keys [[1,0],[237,0],[313,-6.4356],[351,0]]}
       spine_03_ctrl.Rotation.Z {type=MovieSceneScriptingFloatChannel, default=0, thinned_from=5, keys [[1,0],[237,0],[313,-0.2974],[351,0]]}
       spine_03_ctrl.Scale.X {type=MovieSceneScriptingFloatChannel, default=0, const [0,5]}
       spine_03_ctrl.Scale.Y {type=MovieSceneScriptingFloatChannel, default=0, const [0,5]}
       spine_03_ctrl.Scale.Z {type=MovieSceneScriptingFloatChannel, default=0, const [0,5]}
       spine_04_ctrl.Location.X {type=MovieSceneScriptingFloatChannel, default=0, const [0,4]}
       spine_04_ctrl.Location.Y {type=MovieSceneScriptingFloatChannel, default=0, const [0,4]}
       spine_04_ctrl.Location.Z {type=MovieSceneScriptingFloatChannel, default=0, thinned_from=4, keys [[1,0],[352,0]]}
       spine_04_ctrl.Rotation.X {type=MovieSceneScriptingFloatChannel, default=0, keys [[1,6.7949],[237,0],[319,-0.9081],[352,6.7949]]}
       spine_04_ctrl.Rotation.Y {type=MovieSceneScriptingFloatChannel, default=0, keys [[1,1.8579],[237,0],[319,0.8937],[352,1.8579]]}
       spine_04_ctrl.Rotation.Z {type=MovieSceneScriptingFloatChannel, default=0, keys [[1,0.4455],[237,0],[319,-0.3775],[352,0.4455]]}
       spine_04_ctrl.Scale.X {type=MovieSceneScriptingFloatChannel, default=0, const [0,4]}
       spine_04_ctrl.Scale.Y {type=MovieSceneScriptingFloatChannel, default=0, const [0,4]}
       spine_04_ctrl.Scale.Z {type=MovieSceneScriptingFloatChannel, default=0, const [0,4]}
       neck_01_ctrl.Location.X {type=MovieSceneScriptingFloatChannel, default=0, thinned_from=4, keys [[0,0],[351,0]]}
       neck_01_ctrl.Location.Y {type=MovieSceneScriptingFloatChannel, default=0, thinned_from=4, keys [[0,0],[351,0]]}
       neck_01_ctrl.Location.Z {type=MovieSceneScriptingFloatChannel, default=0, thinned_from=4, keys [[0,0],[351,0]]}
       neck_01_ctrl.Rotation.X {type=MovieSceneScriptingFloatChannel, default=0, keys [[0,18.272],[289,14.5414],[318,16.1792],[348,18.272]]}
       neck_01_ctrl.Rotation.Y {type=MovieSceneScriptingFloatChannel, default=0, keys [[0,2.5524],[289,-8.4919],[318,-2.2831],[351,2.5622]]}
       neck_01_ctrl.Rotation.Z {type=MovieSceneScriptingFloatChannel, default=0, keys [[0,0.9693],[289,-0.4387],[318,1.3463],[351,0.9706]]}
       neck_01_ctrl.Scale.X {type=MovieSceneScriptingFloatChannel, default=0, const [0,4]}
       neck_01_ctrl.Scale.Y {type=MovieSceneScriptingFloatChannel, default=0, const [0,4]}
       neck_01_ctrl.Scale.Z {type=MovieSceneScriptingFloatChannel, default=0, const [0,4]}
       Weight {type=MovieSceneScriptingFloatChannel, default=1}
      ]
     }
    ]
   }
  ]
 }
 CineCameraActor {
  id=#2
  possessed_class=CineCameraActor
  tracks [
   {
    type=MovieScene3DTransformTrack
    sections [{
      type=MovieScene3DTransformSection
      channels [
       Location.X {
        type=MovieSceneScriptingDoubleChannel
        default=-10162.821
        thinned_from=6
        keys [[0,-10162.8213],[27,-10167.7173],[141,-10411.6181],[196,-10647.007],[320,-11571.5967]]
       }
       Location.Y {
        type=MovieSceneScriptingDoubleChannel
        default=-501.5211
        keys [[0,-501.5211],[122,-253.1323],[172,-176.4942],[213,-140.8109],[320,-456.3452]]
       }
       Location.Z {type=MovieSceneScriptingDoubleChannel, default=4115.7327, thinned_from=3, keys [[0,4115.7329],[320,4105.6025]]}
       Rotation.X {type=MovieSceneScriptingDoubleChannel, default=0, thinned_from=3, keys [[0,0],[320,0]]}
       Rotation.Y {type=MovieSceneScriptingDoubleChannel, default=0.6, const [0.6,3]}
       Rotation.Z {
        type=MovieSceneScriptingDoubleChannel
        default=-169.0728
        keys [[0,-163.2058],[29,-160.2665],[159,-131.5389],[198,-115.6142],[228,-94.1608],[272,-45.2099],[320,-21.2119]]
       }
       Scale.X {type=MovieSceneScriptingDoubleChannel, default=1, const [1,3]}
       Scale.Y {type=MovieSceneScriptingDoubleChannel, default=1, const [1,3]}
       Scale.Z {type=MovieSceneScriptingDoubleChannel, default=1, const [1,3]}
      ]
     }]
   }
   {type=MovieSceneSpawnTrack, sections [{type=MovieSceneSpawnSection, channels [None {type=MovieSceneScriptingBoolChannel, default=true}]}]}
  ]
 }
 CameraComponent {
  id=#3
  parent_id=#2
  possessed_class=CineCameraComponent
  tracks [
   {
    type=MovieSceneFloatTrack
    display_name=Current Focal Length
    sections [{type=MovieSceneFloatSection, channels [None {type=MovieSceneScriptingFloatChannel, default=35}]}]
   }
   {
    type=MovieSceneFloatTrack
    display_name="Manual Focus Distance (Focus Settings)"
    sections [{type=MovieSceneFloatSection, channels [None {type=MovieSceneScriptingFloatChannel, default=100000}]}]
   }
   {
    type=MovieSceneFloatTrack
    display_name=Current Aperture
    sections [{type=MovieSceneFloatSection, channels [None {type=MovieSceneScriptingFloatChannel, default=2.8}]}]
   }
  ]
 }
]
display_rate {denominator=1, numerator=30}
tick_resolution {denominator=1, numerator=24000}
tracks [{type=MovieSceneCameraCutTrack, sections [{type=MovieSceneCameraCutSection, end_frame=320, start_frame=0}]}]
