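# Test Level Sequence for DragonCart narrative, named for Dark Voyager, that animates a spawnable FortPlayerMannequin placed at a fixed world transform.
# A skeletal animation track sits alongside a Fortnite_Biped_CtrlRig parameter track carrying densely keyed, apparently baked, control motion.
# Visible left-leg FK curves largely repeat a roughly 325-frame cycle, then vary in the later frames before returning near the opening pose.
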LevelSequence /DragonCartNarrative/Sequence/Test_Sequences/DarkVoyager_Test_2
bindings [FortPlayerMannequin {
  id=#1
  possessed_class=FortPlayerMannequin
  tracks [
   {type=MovieSceneSpawnTrack, sections [{type=MovieSceneSpawnSection, channels [None {type=MovieSceneScriptingBoolChannel, default=true}]}]}
   {type=MovieSceneSkeletalAnimationTrack, sections [{type=MovieSceneSkeletalAnimationSection, end_frame=400, start_frame=0}]}
   {
    type=MovieScene3DTransformTrack
    sections [{
      type=MovieScene3DTransformSection
      channels [
       Location.X {type=MovieSceneScriptingDoubleChannel, default=-10942.4065}
       Location.Y {type=MovieSceneScriptingDoubleChannel, default=-693.8637}
       Location.Z {type=MovieSceneScriptingDoubleChannel, default=3990.0013}
       Rotation.X {type=MovieSceneScriptingDoubleChannel, default=0}
       Rotation.Y {type=MovieSceneScriptingDoubleChannel, default=0}
       Rotation.Z {type=MovieSceneScriptingDoubleChannel, default=0}
       Scale.X {type=MovieSceneScriptingDoubleChannel, default=1}
       Scale.Y {type=MovieSceneScriptingDoubleChannel, default=1}
       Scale.Z {type=MovieSceneScriptingDoubleChannel, default=1}
      ]
     }]
   }
   {
    type=MovieSceneControlRigParameterTrack
    display_name=Fortnite_Biped_CtrlRig
    sections [{
      type=MovieSceneControlRigParameterSection
      channels [
       leg_l_stretch_switch {type=MovieSceneScriptingBoolChannel, default=false, const [false,751]}
       leg_r_stretch_switch {type=MovieSceneScriptingBoolChannel, default=false, const [false,751]}
       arm_l_fk_local {type=MovieSceneScriptingBoolChannel, default=true, const [true,636]}
       Auto Interpolate_3 {type=MovieSceneScriptingBoolChannel, default=false, const [false,629]}
       Auto Interpolate {type=MovieSceneScriptingBoolChannel, default=false, const [false,635]}
       arm_l_stretch_switch {type=MovieSceneScriptingBoolChannel, default=false, const [false,751]}
       Keep Aligned {type=MovieSceneScriptingBoolChannel, default=false, const [false,751]}
       arm_r_fk_local {type=MovieSceneScriptingBoolChannel, default=true, const [true,610]}
       Auto Interpolate_4 {type=MovieSceneScriptingBoolChannel, default=false, const [false,630]}
       Auto Interpolate_2 {type=MovieSceneScriptingBoolChannel, default=false, const [false,635]}
       arm_r_stretch_switch {type=MovieSceneScriptingBoolChannel, default=false, const [false,751]}
       Keep Aligned_2 {type=MovieSceneScriptingBoolChannel, default=false, const [false,751]}
       head_local {type=MovieSceneScriptingBoolChannel, default=true, const [true,751]}
       Stretch {type=MovieSceneScriptingBoolChannel, default=false, const [false,751]}
       head_stretch {type=MovieSceneScriptingBoolChannel, default=false, const [false,751]}
       spine_stretch_switch {type=MovieSceneScriptingBoolChannel, default=false, const [false,751]}
       arm_l_fk_ik_switch {type=MovieSceneScriptingBoolChannel, default=false, const [false,757]}
       leg_l_fk_ik_switch {type=MovieSceneScriptingBoolChannel, default=true, const [false,3]}
       arm_r_fk_ik_switch {type=MovieSceneScriptingBoolChannel, default=false, const [false,757]}
       leg_r_fk_ik_switch {type=MovieSceneScriptingBoolChannel, default=false}
       ShowBodyControls {type=MovieSceneScriptingBoolChannel, default=true, const [true,757]}
       spine_fk_ik_switch {type=MovieSceneScriptingBoolChannel, default=false, const [false,757]}
       neck_fk_ik_switch {type=MovieSceneScriptingBoolChannel, default=false, const [false,757]}
       space {type=MovieSceneScriptingByteChannel, default=1, const [1,751]}
       global_ctrl.Location.X {type=MovieSceneScriptingFloatChannel, default=0, const [0,757]}
       global_ctrl.Location.Y {type=MovieSceneScriptingFloatChannel, default=0, const [0,757]}
       global_ctrl.Location.Z {type=MovieSceneScriptingFloatChannel, default=0, const [0,757]}
       global_ctrl.Rotation.X {type=MovieSceneScriptingFloatChannel, default=0, const [0,757]}
       global_ctrl.Rotation.Y {type=MovieSceneScriptingFloatChannel, default=0, const [0,757]}
       global_ctrl.Rotation.Z {type=MovieSceneScriptingFloatChannel, default=0, const [0,757]}
       root_ctrl.Location.X {type=MovieSceneScriptingFloatChannel, default=0, const [0,757]}
       root_ctrl.Location.Y {type=MovieSceneScriptingFloatChannel, default=0, const [0,757]}
       root_ctrl.Location.Z {type=MovieSceneScriptingFloatChannel, default=0, const [0,757]}
       root_ctrl.Rotation.X {type=MovieSceneScriptingFloatChannel, default=0, const [0,757]}
       root_ctrl.Rotation.Y {type=MovieSceneScriptingFloatChannel, default=0, const [0,757]}
       root_ctrl.Rotation.Z {type=MovieSceneScriptingFloatChannel, default=0, const [0,757]}
       root_ctrl.Scale.X {type=MovieSceneScriptingFloatChannel, default=1, const [1,757]}
       root_ctrl.Scale.Y {type=MovieSceneScriptingFloatChannel, default=1, const [1,757]}
       root_ctrl.Scale.Z {type=MovieSceneScriptingFloatChannel, default=1, const [1,757]}
       body_offset_ctrl.Location.X {type=MovieSceneScriptingFloatChannel, default=0, const [0,757]}
       body_offset_ctrl.Location.Y {type=MovieSceneScriptingFloatChannel, default=0, const [0,757]}
       body_offset_ctrl.Location.Z {type=MovieSceneScriptingFloatChannel, default=0, const [0,757]}
       body_offset_ctrl.Rotation.X {type=MovieSceneScriptingFloatChannel, default=0, const [0,757]}
       body_offset_ctrl.Rotation.Y {type=MovieSceneScriptingFloatChannel, default=0, const [0,757]}
       body_offset_ctrl.Rotation.Z {type=MovieSceneScriptingFloatChannel, default=0, const [0,757]}
       body_offset_ctrl.Scale.X {type=MovieSceneScriptingFloatChannel, default=1, const [1,757]}
       body_offset_ctrl.Scale.Y {type=MovieSceneScriptingFloatChannel, default=1, const [1,757]}
       body_offset_ctrl.Scale.Z {type=MovieSceneScriptingFloatChannel, default=1, const [1,757]}
       thigh_l_fk_ctrl.Location.X {type=MovieSceneScriptingFloatChannel, default=0, thinned_from=748, keys [[-2,0],[929,0]]}
       thigh_l_fk_ctrl.Location.Y {type=MovieSceneScriptingFloatChannel, default=0, thinned_from=748, keys [[-2,0],[929,0]]}
       thigh_l_fk_ctrl.Location.Z {type=MovieSceneScriptingFloatChannel, default=0, thinned_from=748, keys [[-2,0],[929,0]]}
       thigh_l_fk_ctrl.Rotation.X {
        type=MovieSceneScriptingFloatChannel
        default=1.216
        thinned_from=318
        keys [[-2,35.8585],[0,35.9548],[2,36.1666],[14,39.117],[22,40.6842],[79,40.3808],[94,39.8963],[145,38.3983],[159,37.4969],[203,35.5453],[249,30.9311],[323,35.8585],[325,35.9548],[327,36.1666],[339,39.117],[347,40.6842],[404,40.3808],[419,39.8963],[470,38.3983],[484,37.4969],[528,35.5453],[574,30.9311],[605,27.9502],[614,27.0036],[623,26.6202],[634,26.8532],[676,25.2831],[680,25.2332],[692,24.5929],[715,22.4367],[728,21.3948],[736,21.045],[757,21.5109],[760,21.4993],[798,20.9966],[832,24.1667],[845,25.2831],[903,27.2869],[926,35.5661],[929,35.8585]]
       }
       thigh_l_fk_ctrl.Rotation.Y {
        type=MovieSceneScriptingFloatChannel
        default=-1.2128
        thinned_from=427
        keys [[-2,-1.6495],[0,-1.5402],[2,-1.2997],[11,0.5143],[12,0.6882],[13,0.8714],[14,1.0614],[16,1.454],[17,1.6448],[18,1.8267],[19,2.0174],[20,2.1737],[21,2.3191],[22,2.4344],[23,2.5241],[24,2.5832],[25,2.6226],[26,2.6527],[27,2.6761],[28,2.6915],[29,2.699],[31,2.6988],[33,2.6731],[35,2.6341],[36,2.6103],[37,2.5799],[38,2.5392],[39,2.4918],[41,2.3751],[42,2.3077],[44,2.1606],[49,1.7614],[53,1.4252],[55,1.2623],[56,1.183],[57,1.1061],[58,1.0324],[59,0.9652],[60,0.9023],[61,0.8427],[62,0.7912],[63,0.7456],[66,0.6245],[67,0.5851],[68,0.547],[69,0.5103],[70,0.4751],[71,0.3896],[72,0.3057],[73,0.2262],[74,0.1501],[75,0.0767],[76,0.0059],[77,-0.0576],[78,-0.118],[79,-0.1765],[80,-0.2328],[97,-1.8649],[98,-1.9991],[117,-5.6514],[118,-5.739],[126,-6.1429],[127,-6.0971],[135,-5.5779],[138,-5.4214],[147,-5.0619],[152,-4.9301],[155,-4.8871],[156,-4.8799],[157,-4.8986],[158,-4.9307],[201,-5.4935],[202,-5.4789],[216,-3.9596],[236,-0.2771],[249,1.5419],[265,2.845],[266,2.9001],[267,2.9259],[268,2.9353],[272,2.9339],[274,2.9482],[275,2.9616],[277,3.0138],[279,3.0792],[282,3.1575],[284,3.1996],[287,3.2362],[289,3.2428],[296,3.2195],[297,3.212],[298,3.1956],[299,3.1701],[300,3.1377],[302,3.0527],[304,2.9433],[305,2.8823],[307,2.7417],[308,2.6643],[309,2.5815],[323,-0.8088],[325,-0.7359],[327,-0.5606],[336,1.112],[337,1.2717],[338,1.4401],[339,1.6151],[341,1.9768],[342,2.1525],[343,2.3755],[344,2.6093],[345,2.8013],[346,2.9803],[347,3.1228],[348,3.2342],[349,3.3087],[350,3.3593],[351,3.3985],[352,3.4298],[353,3.4513],[354,3.4633],[356,3.4688],[358,3.4435],[360,3.4021],[361,3.3761],[362,3.3422],[363,3.2958],[364,3.2414],[366,3.1058],[368,2.9425],[370,2.7626],[374,2.3861],[379,1.8941],[381,1.7062],[382,1.6159],[383,1.5296],[384,1.4512],[385,1.3779],[386,1.3087],[387,1.2492],[388,1.1969],[392,1.0142],[394,0.9294],[395,0.8898],[396,0.7891],[397,0.6904],[398,0.597],[399,0.5079],[400,0.4219],[401,0.3391],[402,0.2651],[403,0.1948],[404,0.1269],[405,0.0617],[422,-1.8649],[423,-1.9991],[442,-5.6514],[443,-5.739],[451,-6.1429],[452,-6.0971],[460,-5.5779],[463,-5.4214],[472,-5.0619],[477,-4.9301],[480,-4.8871],[481,-4.8799],[482,-4.8986],[483,-4.9307],[526,-5.4935],[527,-5.4789],[541,-3.9596],[561,-0.2771],[574,1.5419],[590,2.845],[591,2.9001],[592,2.9259],[593,2.9353],[597,2.9339],[599,2.9482],[600,2.9616],[602,3.0138],[604,3.0792],[607,3.1575],[609,3.1996],[612,3.2362],[614,3.2428],[621,3.2195],[622,3.212],[623,3.1956],[624,3.1701],[625,3.1377],[627,3.0527],[629,2.9433],[630,2.8823],[632,2.7417],[633,2.6643],[634,2.5815],[676,10.2512],[700,13.7245],[716,13.6886],[719,13.5954],[723,13.3404],[728,12.8594],[746,10.5498],[750,10.1112],[753,9.8426],[755,9.7009],[757,9.5941],[760,9.4595],[770,9.3473],[773,9.7401],[777,10.165],[780,10.4161],[790,10.9346],[798,11.1802],[803,11.2369],[820,11.06],[832,10.8461],[845,10.2512],[895,2.4941],[897,2.3032],[898,2.202],[900,1.9845],[902,1.7508],[903,1.6281],[926,-1.5673],[929,-1.6495]]
       }
       thigh_l_fk_ctrl.Rotation.Z {
        type=MovieSceneScriptingFloatChannel
        default=-0.1664
        thinned_from=632
        keys [[-2,10.9928],[0,11.04],[2,11.144],[14,12.3568],[19,12.6007],[27,11.6502],[31,11.134],[37,10.5232],[43,10.026],[47,9.7693],[49,9.6815],[52,9.6292],[55,9.6628],[57,9.7434],[59,9.8686],[61,10.0352],[65,10.4695],[80,12.4054],[82,12.65],[83,12.8165],[85,13.2378],[91,14.7385],[93,15.1347],[94,15.2789],[95,15.3744],[96,15.4256],[97,15.4364],[99,15.3581],[100,15.2684],[101,15.1393],[102,14.9624],[128,13.5735],[131,13.502],[133,13.5144],[135,13.5911],[137,13.7449],[153,15.5771],[158,15.9901],[161,16.2678],[165,16.4507],[168,16.4574],[174,16.6633],[206,16.1916],[230,15.4756],[256,15.5844],[295,14.3054],[303,13.9112],[305,13.7619],[323,10.9928],[325,11.04],[327,11.144],[339,12.3568],[344,12.6007],[352,11.6502],[356,11.134],[362,10.5232],[368,10.026],[372,9.7693],[374,9.6815],[377,9.6292],[380,9.6628],[382,9.7434],[384,9.8686],[386,10.0352],[390,10.4695],[405,12.4054],[407,12.65],[408,12.8165],[410,13.2378],[416,14.7385],[418,15.1347],[419,15.2789],[420,15.3744],[421,15.4256],[422,15.4364],[424,15.3581],[425,15.2684],[426,15.1393],[427,14.9624],[453,13.5735],[456,13.502],[458,13.5144],[460,13.5911],[462,13.7449],[478,15.5771],[483,15.9901],[486,16.2678],[490,16.4507],[493,16.4574],[499,16.6633],[531,16.1916],[555,15.4756],[581,15.5844],[620,14.3054],[628,13.9112],[630,13.7619],[672,0.3666],[676,0.2529],[692,0.1221],[697,0.2953],[698,0.2251],[699,0.1421],[700,0.06],[701,-0.0223],[702,-0.1077],[703,-0.1974],[705,-0.38],[706,-0.4744],[707,-0.567],[709,-0.7468],[710,-0.827],[712,-0.9733],[714,-1.1145],[716,-1.2453],[717,-1.3129],[721,-1.5671],[723,-1.6888],[725,-1.8024],[726,-1.8543],[727,-1.8983],[728,-1.9346],[729,-1.9666],[730,-1.9926],[731,-2.0127],[732,-2.0247],[733,-2.0318],[734,-2.0299],[735,-2.0196],[736,-2.0028],[737,-1.9775],[738,-1.9456],[739,-1.9066],[741,-1.8088],[743,-1.69],[744,-1.6231],[745,-1.5528],[747,-1.4034],[748,-1.3256],[750,-1.1635],[751,-1.0753],[752,-0.978],[753,-0.8691],[754,-0.7489],[755,-0.6202],[757,-0.3364],[759,-0.0191],[760,-0.0245],[781,-0.6078],[786,-0.4552],[787,-0.4255],[788,-0.3967],[789,-0.369],[790,-0.3428],[791,-0.3181],[792,-0.2955],[793,-0.2751],[794,-0.2572],[795,-0.2422],[796,-0.2301],[797,-0.2216],[798,-0.2166],[799,-0.2168],[800,-0.2218],[801,-0.2295],[802,-0.2385],[803,-0.2492],[832,-0.6249],[852,-0.8471],[869,1.1036],[902,14.229],[908,14.1682],[910,14.0804],[926,11.1663],[929,10.9928]]
       }
       thigh_l_fk_ctrl.Scale.X {type=MovieSceneScriptingFloatChannel, default=1, const [1,748]}
       thigh_l_fk_ctrl.Scale.Y {type=MovieSceneScriptingFloatChannel, default=1, const [1,748]}
       thigh_l_fk_ctrl.Scale.Z {type=MovieSceneScriptingFloatChannel, default=1, const [1,748]}
       calf_l_fk_ctrl.Location.X {type=MovieSceneScriptingFloatChannel, default=0, thinned_from=751, keys [[-2,0.013],[929,0.013]]}
       calf_l_fk_ctrl.Location.Y {type=MovieSceneScriptingFloatChannel, default=0, thinned_from=751, keys [[-2,-0.0001],[929,-0.0001]]}
       calf_l_fk_ctrl.Location.Z {type=MovieSceneScriptingFloatChannel, default=0, thinned_from=751, keys [[-2,0],[929,0]]}
       calf_l_fk_ctrl.Rotation.X {
        type=MovieSceneScriptingFloatChannel
        default=0.0543
        thinned_from=751
        keys [[-2,0.052],[-1,0.065],[0,0.0618],[1,0.0578],[2,0.0717],[3,0.0799],[4,0.0872],[5,0.0931],[6,0.098],[8,0.1068],[10,0.1149],[17,0.1414],[18,0.145],[19,0.1606],[24,0.1768],[26,0.1824],[28,0.187],[31,0.192],[34,0.1959],[43,0.2023],[47,0.2039],[48,0.2037],[49,0.2043],[51,0.2039],[52,0.2044],[56,0.204],[57,0.2033],[58,0.2033],[68,0.1963],[74,0.19],[76,0.1873],[79,0.1819],[85,0.1702],[86,0.1675],[87,0.1643],[88,0.1605],[89,0.1559],[90,0.1506],[91,0.1444],[92,0.1372],[93,0.129],[94,0.1195],[95,0.1086],[96,0.0963],[97,0.0823],[98,0.0663],[99,0.0476],[100,0.0265],[101,-0.0071],[102,-0.0069],[103,-0.0496],[104,-0.0785],[105,-0.1087],[106,-0.1405],[107,-0.1764],[108,-0.2449],[110,-0.2479],[120,-0.2672],[121,-0.2205],[122,-0.1998],[123,-0.1841],[124,-0.1716],[125,-0.1622],[126,-0.1554],[128,-0.1448],[130,-0.1351],[132,-0.1268],[134,-0.119],[135,-0.1154],[137,-0.1086],[140,-0.0991],[142,-0.0937],[145,-0.0863],[148,-0.0796],[152,-0.0719],[155,-0.0669],[156,-0.0656],[157,-0.0638],[158,-0.0619],[160,-0.0585],[162,-0.056],[163,-0.0554],[164,-0.0541],[166,-0.0528],[167,-0.0528],[168,-0.0522],[169,-0.0525],[173,-0.0525],[174,-0.052],[175,-0.0523],[176,-0.0519],[177,-0.052],[178,-0.0515],[179,-0.0517],[181,-0.0509],[182,-0.0513],[184,-0.0506],[185,-0.0508],[197,-0.0475],[198,-0.0465],[199,-0.0463],[200,-0.0456],[202,-0.0455],[203,-0.0457],[205,-0.0477],[206,-0.0489],[210,-0.0555],[211,-0.0556],[212,-0.0525],[213,-0.0464],[214,-0.0377],[215,-0.0115],[217,-0.0122],[218,0.0173],[219,0.0344],[220,0.0528],[221,0.0722],[222,0.0925],[223,0.1135],[224,0.135],[225,0.1569],[231,0.2913],[232,0.3132],[233,0.3343],[234,0.3546],[235,0.374],[236,0.3922],[237,0.409],[238,0.4242],[239,0.4376],[240,0.4493],[241,0.4589],[242,0.4664],[243,0.4715],[245,0.4781],[247,0.482],[251,0.4855],[267,0.4742],[277,0.4617],[290,0.4475],[309,0.4149],[323,0.052],[324,0.065],[325,0.0618],[326,0.0578],[327,0.0717],[328,0.0799],[329,0.0872],[330,0.0931],[331,0.098],[333,0.1068],[335,0.1149],[342,0.1414],[343,0.145],[344,0.1606],[349,0.1768],[351,0.1824],[353,0.187],[356,0.192],[359,0.1959],[368,0.2023],[372,0.2039],[373,0.2037],[374,0.2043],[376,0.2039],[377,0.2044],[381,0.204],[382,0.2033],[383,0.2033],[393,0.1963],[399,0.19],[401,0.1873],[404,0.1819],[410,0.1702],[411,0.1675],[412,0.1643],[413,0.1605],[414,0.1559],[415,0.1506],[416,0.1444],[417,0.1372],[418,0.129],[419,0.1195],[420,0.1086],[421,0.0963],[422,0.0823],[423,0.0663],[424,0.0476],[425,0.0265],[426,-0.0071],[427,-0.0069],[428,-0.0496],[429,-0.0785],[430,-0.1087],[431,-0.1405],[432,-0.1764],[433,-0.2449],[435,-0.2479],[445,-0.2672],[446,-0.2205],[447,-0.1998],[448,-0.1841],[449,-0.1716],[450,-0.1622],[451,-0.1554],[453,-0.1448],[455,-0.1351],[457,-0.1268],[459,-0.119],[460,-0.1154],[462,-0.1086],[465,-0.0991],[467,-0.0937],[470,-0.0863],[473,-0.0796],[477,-0.0719],[480,-0.0669],[481,-0.0656],[482,-0.0638],[483,-0.0619],[485,-0.0585],[487,-0.056],[488,-0.0554],[489,-0.0541],[491,-0.0528],[492,-0.0528],[493,-0.0522],[494,-0.0525],[498,-0.0525],[499,-0.052],[500,-0.0523],[501,-0.0519],[502,-0.052],[503,-0.0515],[504,-0.0517],[506,-0.0509],[507,-0.0513],[509,-0.0506],[510,-0.0508],[522,-0.0475],[523,-0.0465],[524,-0.0463],[525,-0.0456],[527,-0.0455],[528,-0.0457],[530,-0.0477],[531,-0.0489],[535,-0.0555],[536,-0.0556],[537,-0.0525],[538,-0.0464],[539,-0.0377],[540,-0.0115],[542,-0.0122],[543,0.0173],[544,0.0344],[545,0.0528],[546,0.0722],[547,0.0925],[548,0.1135],[549,0.135],[550,0.1569],[556,0.2913],[557,0.3132],[558,0.3343],[559,0.3546],[560,0.374],[561,0.3922],[562,0.409],[563,0.4242],[564,0.4376],[565,0.4493],[566,0.4589],[567,0.4664],[568,0.4715],[570,0.4781],[572,0.482],[576,0.4855],[592,0.4742],[602,0.4617],[615,0.4475],[634,0.4149],[676,0.026],[677,0.0389],[678,0.0357],[679,0.0317],[680,0.0456],[681,0.0538],[682,0.0611],[683,0.067],[684,0.0719],[686,0.0807],[688,0.0888],[693,0.108],[696,0.1189],[697,0.1345],[701,0.1476],[703,0.1537],[705,0.1587],[707,0.1628],[711,0.1687],[715,0.1725],[725,0.1793],[726,0.1793],[727,0.1802],[729,0.1805],[730,0.1813],[734,0.1823],[735,0.182],[736,0.1823],[745,0.1794],[754,0.1726],[759,0.1666],[760,0.1728],[770,0.4328],[800,0.3833],[803,0.3768],[832,0.0968],[845,0.026],[895,0.4133],[903,0.3987],[926,0.0639],[929,0.052]]
       }
       calf_l_fk_ctrl.Rotation.Y {
        type=MovieSceneScriptingFloatChannel
        default=1.2238
        thinned_from=751
        keys [[-2,-0.0189],[-1,-0.0178],[1,-0.0196],[2,-0.0201],[83,-0.0271],[95,-0.0259],[100,-0.025],[102,-0.025],[104,-0.0255],[106,-0.0265],[107,-0.0273],[108,-0.0286],[120,-0.0334],[121,-0.0331],[158,-0.0378],[206,-0.0382],[227,-0.041],[233,-0.0409],[241,-0.039],[250,-0.0363],[309,-0.0277],[323,-0.0189],[324,-0.0178],[326,-0.0196],[327,-0.0201],[408,-0.0271],[420,-0.0259],[425,-0.025],[427,-0.025],[429,-0.0255],[431,-0.0265],[432,-0.0273],[433,-0.0286],[445,-0.0334],[446,-0.0331],[483,-0.0378],[531,-0.0382],[552,-0.041],[558,-0.0409],[566,-0.039],[575,-0.0363],[634,-0.0277],[676,-0.0489],[677,-0.0478],[679,-0.0496],[680,-0.0501],[716,-0.0539],[718,-0.0534],[720,-0.0524],[722,-0.0509],[725,-0.0477],[728,-0.0437],[733,-0.0356],[744,-0.016],[747,-0.0112],[750,-0.007],[751,-0.0059],[752,-0.0049],[754,-0.0035],[755,-0.0031],[757,-0.0028],[759,-0.0034],[760,-0.0049],[770,-0.0396],[783,-0.0303],[787,-0.0283],[790,-0.0276],[792,-0.0277],[794,-0.0282],[796,-0.0292],[798,-0.0309],[803,-0.0358],[832,-0.0519],[845,-0.0489],[895,-0.0275],[903,-0.0259],[926,-0.0191],[929,-0.0189]]
       }
       calf_l_fk_ctrl.Rotation.Z {
        type=MovieSceneScriptingFloatChannel
        default=0.3732
        thinned_from=291
        keys [[-2,8.6668],[0,5.3601],[19,4.4227],[33,4.705],[48,4.5807],[66,4.7182],[90,8.3573],[104,10.9543],[121,11.5104],[124,11.3369],[184,-0.96],[190,-0.272],[206,2.7158],[210,3.5595],[212,3.9996],[215,4.6954],[225,7.1989],[228,7.9054],[234,9.2562],[236,9.6493],[237,9.8128],[238,9.9492],[239,10.0522],[240,10.1184],[241,10.1413],[242,10.1147],[244,9.9682],[247,9.6801],[250,9.334],[253,8.9283],[256,8.4569],[259,7.9145],[267,6.338],[270,5.7868],[273,5.2808],[275,4.9689],[277,4.6824],[278,4.552],[282,4.0715],[285,3.6976],[287,3.4702],[289,3.2653],[290,3.1728],[291,3.0868],[292,3.0091],[293,2.9396],[294,2.8787],[295,2.8275],[303,3.4859],[305,3.6892],[339,4.8036],[358,4.705],[373,4.5807],[391,4.7182],[415,8.3573],[429,10.9543],[446,11.5104],[449,11.3369],[509,-0.96],[515,-0.272],[531,2.7158],[535,3.5595],[537,3.9996],[540,4.6954],[550,7.1989],[553,7.9054],[559,9.2562],[561,9.6493],[562,9.8128],[563,9.9492],[564,10.0522],[565,10.1184],[566,10.1413],[567,10.1147],[569,9.9682],[572,9.6801],[575,9.334],[578,8.9283],[581,8.4569],[584,7.9145],[592,6.338],[595,5.7868],[598,5.2808],[600,4.9689],[602,4.6824],[603,4.552],[607,4.0715],[610,3.6976],[612,3.4702],[614,3.2653],[615,3.1728],[616,3.0868],[617,3.0091],[618,2.9396],[619,2.8787],[620,2.8275],[628,3.4859],[630,3.6892],[672,24.1535],[680,25.3123],[692,27.3626],[705,28.6362],[713,28.9515],[716,28.9262],[718,28.7774],[721,28.3272],[724,27.6365],[727,26.7422],[731,25.2876],[737,22.7266],[747,18.1638],[751,16.4545],[759,14.2757],[773,12.7677],[790,20.3093],[800,26.8371],[802,27.966],[803,28.4857],[812,31.0861],[819,31.1463],[851,24.9818],[852,24.9299],[856,24.5848],[869,23.8911],[894,12.9144],[902,9.9629],[904,9.7605],[906,9.4981],[909,9.0156],[929,5.5601],[930,5.3492]]
       }
       calf_l_fk_ctrl.Scale.X {type=MovieSceneScriptingFloatChannel, default=1, const [1,751]}
       calf_l_fk_ctrl.Scale.Y {type=MovieSceneScriptingFloatChannel, default=1, const [1,751]}
       calf_l_fk_ctrl.Scale.Z {type=MovieSceneScriptingFloatChannel, default=1, const [1,751]}
       foot_l_fk_ctrl.Location.X {type=MovieSceneScriptingFloatChannel, default=0, thinned_from=747, keys [[-2,0.0542],[929,0.0542]]}
       foot_l_fk_ctrl.Location.Y {type=MovieSceneScriptingFloatChannel, default=0.0398, thinned_from=747, keys [[-2,0.0001],[929,0.0001]]}
       foot_l_fk_ctrl.Location.Z {type=MovieSceneScriptingFloatChannel, default=-0.8885, thinned_from=747, keys [[-2,0],[929,0]]}
       foot_l_fk_ctrl.Rotation.X {
        type=MovieSceneScriptingFloatChannel
        default=-1.1647
        thinned_from=617
        keys [[-2,-3.9261],[0,-4.0334],[5,-4.5933],[9,-4.6945],[11,-4.6926],[12,-4.6828],[13,-4.6537],[14,-4.6031],[15,-4.5358],[16,-4.4487],[17,-4.3472],[18,-4.2339],[21,-3.8628],[23,-3.6033],[24,-3.4808],[25,-3.3664],[26,-3.2634],[27,-3.1682],[28,-3.0951],[29,-3.0343],[30,-2.9959],[31,-2.9854],[32,-2.9996],[33,-3.051],[34,-3.1332],[35,-3.2414],[36,-3.3795],[37,-3.5377],[38,-3.7163],[39,-3.9105],[40,-4.1185],[41,-4.3367],[43,-4.7916],[44,-5.0298],[47,-5.7038],[49,-6.1183],[51,-6.4553],[52,-6.6026],[53,-6.7218],[54,-6.8091],[55,-6.8825],[57,-6.9881],[59,-7.0445],[61,-7.0601],[63,-7.032],[65,-6.9781],[68,-6.8368],[80,-6.0605],[85,-5.6727],[86,-5.7333],[87,-6.0505],[88,-6.5716],[89,-7.2534],[90,-8.0652],[91,-8.9542],[93,-10.8064],[94,-11.687],[95,-12.4808],[96,-13.1434],[97,-13.6397],[98,-13.9325],[99,-13.9944],[100,-13.9066],[102,-13.656],[108,-12.5038],[110,-12.0476],[112,-11.6216],[120,-9.6261],[122,-9.0756],[125,-8.322],[127,-7.8663],[128,-7.6715],[129,-7.5025],[130,-7.3547],[131,-7.2408],[132,-7.1569],[133,-7.1028],[134,-7.0874],[135,-7.1098],[139,-7.2585],[147,-7.453],[151,-7.4781],[155,-7.4311],[156,-7.4073],[159,-7.2547],[162,-7.0514],[173,-6.1068],[176,-5.8976],[178,-5.785],[183,-5.5752],[197,-5.1866],[201,-5.1057],[203,-5.097],[206,-5.1133],[209,-5.1681],[210,-5.1963],[224,-5.4666],[226,-5.5273],[228,-5.6189],[235,-6.0165],[238,-6.1411],[240,-6.1912],[242,-6.2079],[244,-6.1888],[246,-6.1287],[248,-6.0297],[255,-5.5591],[257,-5.4494],[259,-5.3732],[260,-5.3506],[261,-5.3398],[262,-5.341],[264,-5.381],[268,-5.5264],[276,-5.8644],[288,-6.1741],[296,-6.283],[310,-6.326],[335,-6.5067],[373,-6.4333],[416,-8.9542],[418,-10.8064],[419,-11.687],[420,-12.4808],[421,-13.1434],[422,-13.6397],[423,-13.9325],[424,-13.9944],[425,-13.9066],[427,-13.656],[433,-12.5038],[435,-12.0476],[437,-11.6216],[445,-9.6261],[447,-9.0756],[450,-8.322],[452,-7.8663],[453,-7.6715],[454,-7.5025],[455,-7.3547],[456,-7.2408],[457,-7.1569],[458,-7.1028],[459,-7.0874],[460,-7.1098],[464,-7.2585],[472,-7.453],[476,-7.4781],[480,-7.4311],[481,-7.4073],[484,-7.2547],[487,-7.0514],[498,-6.1068],[501,-5.8976],[503,-5.785],[508,-5.5752],[522,-5.1866],[526,-5.1057],[528,-5.097],[531,-5.1133],[534,-5.1681],[535,-5.1963],[536,-5.2395],[537,-5.308],[538,-5.3946],[539,-5.4926],[540,-5.6118],[541,-5.7164],[542,-5.8346],[544,-6.1245],[546,-6.3853],[548,-6.6171],[549,-6.7506],[550,-6.9362],[551,-7.2247],[552,-7.6025],[553,-8.0514],[554,-8.5626],[555,-9.1111],[558,-10.8114],[559,-11.3374],[560,-11.8213],[561,-12.249],[562,-12.6154],[563,-12.914],[564,-13.1309],[565,-13.2669],[566,-13.3064],[567,-13.2391],[568,-13.0707],[569,-12.8041],[570,-12.4197],[571,-11.9011],[572,-11.2746],[573,-10.5536],[574,-9.7534],[575,-8.8996],[576,-8.0133],[579,-5.2823],[580,-4.4266],[581,-3.6468],[582,-2.9575],[583,-2.3595],[584,-1.8707],[585,-1.5037],[586,-1.2723],[587,-1.1765],[588,-1.2397],[589,-1.4141],[590,-1.6524],[591,-1.9427],[592,-2.2716],[593,-2.6355],[594,-3.0214],[597,-4.2261],[598,-4.6191],[599,-4.9789],[600,-5.3082],[601,-5.6172],[602,-5.9075],[603,-6.1667],[604,-6.3869],[606,-6.7687],[609,-7.2908],[611,-7.5896],[613,-7.8399],[615,-8.0352],[616,-8.1082],[618,-8.2057],[619,-8.2264],[620,-8.2273],[621,-8.2088],[623,-8.1203],[626,-7.9301],[629,-7.6914],[633,-7.2944],[634,-7.1854],[635,-7.1729],[676,-18.0975],[677,-18.2114],[683,-18.6376],[687,-18.4977],[690,-18.2551],[692,-18.0088],[695,-17.4994],[702,-16.1293],[704,-15.8178],[706,-15.5886],[707,-15.5117],[708,-15.4678],[710,-15.4767],[711,-15.5344],[713,-15.7446],[715,-16.0696],[718,-16.7087],[727,-18.9708],[730,-19.5641],[733,-19.9601],[737,-20.2835],[743,-20.4625],[750,-20.3327],[751,-20.2786],[752,-20.1837],[754,-19.875],[755,-19.6677],[757,-19.1489],[759,-18.4888],[760,-18.4902],[798,-19.3123],[803,-19.3279],[832,-18.3659],[837,-18.0962],[858,-14.5858],[895,-5.8288],[903,-5.1577],[929,-3.9261]]
       }
       foot_l_fk_ctrl.Rotation.Y {
        type=MovieSceneScriptingFloatChannel
        default=0.0792
        thinned_from=735
        keys [[-2,10.3797],[0,10.352],[4,10.2301],[12,10.4116],[14,10.4945],[16,10.6184],[18,10.7794],[19,10.8974],[26,11.5186],[29,11.7081],[30,11.7536],[32,11.7843],[34,11.7472],[36,11.6438],[39,11.3946],[49,10.2586],[52,9.9915],[54,9.87],[56,9.7834],[59,9.7007],[64,9.6585],[82,9.967],[85,10.0443],[88,10.0195],[95,9.8209],[152,9.0396],[233,9.9277],[243,9.8911],[249,11.1008],[251,11.4619],[253,11.7661],[254,11.8994],[255,12.0028],[256,12.0583],[257,12.0766],[258,12.0622],[259,12.0219],[261,11.8587],[263,11.6256],[270,10.594],[272,10.3385],[273,10.2267],[275,10.0733],[282,9.6724],[286,9.515],[291,9.4242],[294,9.4349],[299,9.5609],[303,9.7536],[309,10.177],[323,10.3797],[325,10.352],[329,10.2301],[337,10.4116],[339,10.4945],[341,10.6184],[343,10.7794],[344,10.8974],[351,11.5186],[354,11.7081],[355,11.7536],[357,11.7843],[359,11.7472],[361,11.6438],[364,11.3946],[374,10.2586],[377,9.9915],[379,9.87],[381,9.7834],[384,9.7007],[389,9.6585],[407,9.967],[410,10.0443],[413,10.0195],[420,9.8209],[477,9.0396],[558,9.9277],[568,9.8911],[574,11.1008],[576,11.4619],[578,11.7661],[579,11.8994],[580,12.0028],[581,12.0583],[582,12.0766],[583,12.0622],[584,12.0219],[586,11.8587],[588,11.6256],[595,10.594],[597,10.3385],[598,10.2267],[600,10.0733],[607,9.6724],[611,9.515],[616,9.4242],[619,9.4349],[624,9.5609],[628,9.7536],[634,10.177],[635,10.1876],[676,7.5861],[677,7.5603],[682,7.4798],[683,7.5223],[688,7.7948],[691,8.0051],[694,8.3079],[696,8.555],[697,8.7144],[702,9.3453],[705,9.6539],[707,9.7897],[708,9.838],[710,9.8661],[712,9.8212],[714,9.7051],[716,9.5333],[720,9.0671],[727,8.1463],[729,7.9216],[731,7.7329],[734,7.5344],[737,7.3888],[740,7.2864],[744,7.2092],[750,7.1928],[752,7.222],[754,7.2976],[755,7.3499],[757,7.4854],[759,7.6569],[770,7.627],[774,7.5292],[778,7.4882],[782,7.5167],[785,7.5852],[788,7.7044],[803,8.5932],[832,7.8163],[845,7.5868],[858,7.9764],[895,10.8988],[903,11.3861],[929,10.3797]]
       }
       foot_l_fk_ctrl.Rotation.Z {
        type=MovieSceneScriptingFloatChannel
        default=-0.2074
        thinned_from=482
        keys [[-2,31.5119],[0,31.5482],[15,27.6049],[22,24.015],[24,23.0908],[26,22.2671],[28,21.5701],[30,21.0273],[32,20.6388],[34,20.3784],[36,20.2386],[38,20.2097],[40,20.2838],[42,20.451],[45,20.8583],[48,21.4256],[53,22.6424],[74,29.0137],[79,30.2569],[83,30.973],[86,31.2643],[89,31.3175],[92,31.1204],[94,30.8687],[108,28.3075],[134,19.4372],[141,17.2228],[144,16.3689],[150,15.0012],[154,14.2278],[159,13.3933],[161,13.0057],[164,12.3132],[169,11.0682],[171,10.6246],[172,10.4243],[173,10.2466],[174,10.0975],[175,9.9794],[176,9.9076],[177,9.8795],[178,9.9005],[179,9.97],[180,10.0808],[181,10.2264],[182,10.3999],[183,10.6005],[184,10.8274],[186,11.3569],[187,11.6519],[189,12.3106],[191,13.031],[195,14.6285],[196,15.0501],[204,17.7466],[217,21.838],[221,22.9066],[225,23.7348],[229,24.3404],[233,24.7436],[236,24.8698],[239,24.8308],[243,24.5373],[245,24.3078],[268,20.8348],[298,26.5494],[342,17.9615],[365,20.2838],[366,20.3558],[369,20.7032],[372,21.2204],[376,22.1217],[400,29.2841],[405,30.4636],[408,30.973],[411,31.2643],[414,31.3175],[417,31.1204],[419,30.8687],[433,28.3075],[459,19.4372],[466,17.2228],[469,16.3689],[475,15.0012],[479,14.2278],[484,13.3933],[486,13.0057],[489,12.3132],[494,11.0682],[496,10.6246],[497,10.4243],[498,10.2466],[499,10.0975],[500,9.9794],[501,9.9076],[502,9.8795],[503,9.9005],[504,9.97],[505,10.0808],[506,10.2264],[507,10.3999],[508,10.6005],[509,10.8274],[511,11.3569],[512,11.6519],[514,12.3106],[516,13.031],[520,14.6285],[521,15.0501],[529,17.7466],[542,21.838],[546,22.9066],[550,23.7348],[554,24.3404],[558,24.7436],[561,24.8698],[564,24.8308],[568,24.5373],[570,24.3078],[594,19.444],[613,15.7158],[617,15.1992],[628,13.43],[630,13.1356],[672,11.5497],[677,11.7954],[693,11.6983],[716,12.0755],[718,12.1926],[720,12.3777],[730,13.709],[733,14.0226],[752,15.2968],[754,15.5091],[757,15.9424],[759,16.282],[764,16.1777],[777,10.403],[779,9.2984],[780,8.7633],[781,8.2496],[782,7.7691],[783,7.3145],[784,6.9042],[785,6.5311],[786,6.2133],[787,5.9567],[788,5.7567],[789,5.6266],[790,5.5712],[791,5.578],[792,5.6302],[793,5.7231],[794,5.8514],[795,6.0075],[796,6.195],[797,6.3973],[798,6.6182],[799,6.8524],[800,7.101],[803,7.9111],[816,11.1358],[850,10.5792],[852,10.6366],[869,13.4228],[902,31.2449],[910,30.5743],[929,31.5119]]
       }
       foot_l_fk_ctrl.Scale.X {type=MovieSceneScriptingFloatChannel, default=1, const [1,747]}
       foot_l_fk_ctrl.Scale.Y {type=MovieSceneScriptingFloatChannel, default=1, const [1,747]}
       foot_l_fk_ctrl.Scale.Z {type=MovieSceneScriptingFloatChannel, default=1, const [1,747]}
       ball_l_fk_ctrl.Location.X {type=MovieSceneScriptingFloatChannel, default=0, thinned_from=751, keys [[-2,0],[929,0]]}
       ball_l_fk_ctrl.Location.Y {type=MovieSceneScriptingFloatChannel, default=0, thinned_from=751, keys [[-2,0],[929,0]]}
       ball_l_fk_ctrl.Location.Z {type=MovieSceneScriptingFloatChannel, default=0, thinned_from=751, keys [[-2,0],[929,0]]}
       ball_l_fk_ctrl.Rotation.X {
        type=MovieSceneScriptingFloatChannel
        default=0
        thinned_from=751
        keys [[-2,0.0919],[-1,0.0997],[1,0.0997],[2,0.0989],[6,0.0989],[7,0.0998],[12,0.0998],[13,0.1088],[16,0.1089],[17,0.1185],[20,0.1185],[21,0.129],[24,0.129],[25,0.1395],[29,0.1395],[30,0.146],[31,0.1465],[36,0.1465],[37,0.1424],[41,0.1424],[42,0.1324],[45,0.1324],[46,0.1204],[47,0.1209],[49,0.1209],[50,0.109],[53,0.109],[54,0.0995],[62,0.0995],[63,0.1007],[64,0.0916],[85,0.0918],[86,0.0894],[87,0.0894],[88,0.0769],[89,0.0709],[90,0.0609],[91,0.0484],[92,0.0376],[93,0.026],[94,0.0104],[95,0.0118],[96,-0.0101],[97,-0.0101],[98,-0.0265],[103,-0.0265],[104,-0.0195],[106,-0.0195],[107,-0.0074],[108,0.0011],[109,-0.827],[110,-0.8497],[111,-0.8612],[112,-0.8665],[114,-0.8665],[115,-0.8494],[116,-0.8342],[117,-0.8145],[118,-0.791],[119,-0.7632],[120,0.0615],[121,0.0641],[123,0.0641],[124,0.076],[126,0.076],[127,0.0857],[130,0.0857],[131,0.0954],[134,0.0954],[135,0.0992],[142,0.0993],[143,0.0983],[161,0.0984],[162,0.098],[164,0.098],[165,0.0983],[167,0.0983],[168,0.099],[170,0.099],[171,0.0987],[174,0.0987],[193,0.0986],[194,0.1004],[205,0.1003],[206,0.1006],[209,0.1005],[210,0.1007],[217,0.1006],[218,0.1003],[226,0.1002],[227,0.0961],[228,0.0961],[229,0.0819],[230,0.0819],[231,0.0657],[232,0.0657],[233,0.0495],[234,0.0495],[235,0.0313],[236,0.0313],[237,0.0131],[238,0.0131],[239,-0.0018],[240,-0.0018],[241,-0.0122],[247,-0.0122],[248,-0.0017],[249,-0.0017],[250,0.0144],[251,0.0268],[252,0.04],[253,0.04],[254,0.0667],[255,0.0797],[256,0.0797],[257,0.1011],[258,0.1011],[259,0.1208],[260,0.1208],[261,0.1323],[266,0.1323],[267,0.1308],[269,0.1308],[270,0.1181],[271,0.1193],[272,0.1193],[273,0.1018],[275,0.1018],[276,0.0931],[279,0.0931],[280,0.0791],[285,0.0791],[286,0.0617],[287,0.0647],[293,0.0647],[294,0.0514],[302,0.0514],[303,0.0519],[304,0.0516],[309,0.0516],[323,0.0919],[324,0.0997],[326,0.0997],[327,0.0989],[331,0.0989],[332,0.0998],[337,0.0998],[338,0.1088],[341,0.1089],[342,0.1185],[345,0.1185],[346,0.129],[349,0.129],[350,0.1395],[354,0.1395],[355,0.146],[356,0.1465],[361,0.1465],[362,0.1424],[366,0.1424],[367,0.1324],[370,0.1324],[371,0.1204],[372,0.1209],[374,0.1209],[375,0.109],[378,0.109],[379,0.0995],[387,0.0995],[388,0.1007],[389,0.0916],[410,0.0918],[411,0.0894],[412,0.0894],[413,0.0769],[414,0.0709],[415,0.0609],[416,0.0484],[417,0.0376],[418,0.026],[419,0.0104],[420,0.0118],[421,-0.0101],[422,-0.0101],[423,-0.0265],[428,-0.0265],[429,-0.0195],[431,-0.0195],[432,-0.0074],[433,0.0011],[434,-0.827],[435,-0.8497],[436,-0.8612],[437,-0.8665],[439,-0.8665],[440,-0.8494],[441,-0.8342],[442,-0.8145],[443,-0.791],[444,-0.7632],[445,0.0615],[446,0.0641],[448,0.0641],[449,0.076],[451,0.076],[452,0.0857],[455,0.0857],[456,0.0954],[459,0.0954],[460,0.0992],[467,0.0993],[468,0.0983],[486,0.0984],[487,0.098],[489,0.098],[490,0.0983],[492,0.0983],[493,0.099],[495,0.099],[496,0.0987],[499,0.0987],[518,0.0986],[519,0.1004],[530,0.1003],[531,0.1006],[534,0.1005],[535,0.1007],[542,0.1006],[543,0.1003],[551,0.1002],[552,0.0961],[553,0.0961],[554,0.0819],[555,0.0819],[556,0.0657],[557,0.0657],[558,0.0495],[559,0.0495],[560,0.0313],[561,0.0313],[562,0.0131],[563,0.0131],[564,-0.0018],[565,-0.0018],[566,-0.0122],[572,-0.0122],[573,-0.0017],[574,-0.0017],[575,0.0144],[576,0.0268],[577,0.04],[578,0.04],[579,0.0667],[580,0.0797],[581,0.0797],[582,0.1011],[583,0.1011],[584,0.1208],[585,0.1208],[586,0.1323],[591,0.1323],[592,0.1308],[594,0.1308],[595,0.1181],[596,0.1193],[597,0.1193],[598,0.1018],[600,0.1018],[601,0.0931],[604,0.0931],[605,0.0791],[610,0.0791],[611,0.0617],[612,0.0647],[618,0.0647],[619,0.0514],[627,0.0514],[628,0.0519],[629,0.0516],[634,0.0516],[676,0.0919],[677,0.0997],[679,0.0997],[680,0.0989],[684,0.0989],[685,0.0998],[690,0.0998],[691,0.1088],[694,0.1089],[695,0.1185],[698,0.1185],[699,0.129],[702,0.129],[703,0.1395],[707,0.1395],[708,0.146],[709,0.1465],[714,0.1465],[715,0.1424],[719,0.1424],[720,0.1324],[723,0.1324],[724,0.1204],[725,0.1209],[727,0.1209],[728,0.109],[731,0.109],[732,0.0995],[740,0.0995],[741,0.1007],[742,0.0916],[759,0.0917],[770,0.0791],[771,0.0617],[772,0.0647],[778,0.0647],[779,0.0514],[787,0.0514],[788,0.0519],[789,0.0516],[795,0.0516],[796,0.0531],[803,0.0531],[837,0.0919],[887,0.0516],[888,0.0531],[895,0.0531],[929,0.0919]]
       }
       ball_l_fk_ctrl.Rotation.Y {
        type=MovieSceneScriptingFloatChannel
        default=0
        thinned_from=751
        keys [[-2,-0.1053],[-1,-0.108],[1,-0.108],[2,-0.1019],[6,-0.1019],[7,-0.0975],[12,-0.0975],[13,-0.085],[14,-0.085],[15,-0.0872],[16,-0.0872],[17,-0.0728],[20,-0.0728],[21,-0.0591],[24,-0.0591],[25,-0.0455],[29,-0.0455],[30,-0.0342],[31,-0.0346],[36,-0.0346],[37,-0.0386],[41,-0.0386],[42,-0.0521],[45,-0.0521],[46,-0.0677],[47,-0.0661],[49,-0.0661],[50,-0.0831],[53,-0.0831],[54,-0.0952],[62,-0.0952],[63,-0.0922],[64,-0.1064],[78,-0.1063],[79,-0.1061],[85,-0.106],[86,-0.11],[87,-0.11],[88,-0.1303],[89,-0.144],[90,-0.1602],[91,-0.1805],[92,-0.1959],[93,-0.2105],[94,-0.2276],[95,-0.2264],[96,-0.242],[97,-0.242],[98,-0.2462],[103,-0.2462],[104,-0.2331],[106,-0.2331],[107,-0.221],[108,-0.2187],[109,0.4645],[110,0.4536],[111,0.448],[112,0.4471],[114,0.4471],[115,0.4661],[116,0.4787],[117,0.4942],[118,0.5123],[119,0.5316],[120,-0.1528],[121,-0.1473],[123,-0.1473],[124,-0.1326],[126,-0.1326],[127,-0.1199],[130,-0.1199],[131,-0.1071],[134,-0.1071],[135,-0.1041],[142,-0.104],[143,-0.1038],[161,-0.1037],[162,-0.1032],[164,-0.1032],[165,-0.103],[170,-0.1031],[171,-0.1026],[174,-0.1026],[175,-0.1014],[190,-0.1014],[205,-0.1014],[206,-0.101],[209,-0.1011],[210,-0.1009],[217,-0.101],[218,-0.1013],[222,-0.1013],[223,-0.1002],[226,-0.1002],[227,-0.1043],[228,-0.1043],[229,-0.1062],[230,-0.1062],[231,-0.1141],[232,-0.1141],[233,-0.1221],[234,-0.1221],[235,-0.1279],[236,-0.1279],[237,-0.1337],[238,-0.1337],[239,-0.1401],[240,-0.1401],[241,-0.1445],[247,-0.1445],[248,-0.14],[249,-0.14],[251,-0.1275],[252,-0.1222],[253,-0.1222],[254,-0.1116],[255,-0.1091],[256,-0.1091],[257,-0.1004],[258,-0.1004],[259,-0.095],[260,-0.095],[261,-0.0915],[266,-0.0915],[267,-0.0898],[269,-0.0898],[270,-0.0945],[272,-0.0946],[273,-0.098],[275,-0.098],[276,-0.0997],[279,-0.0997],[280,-0.1066],[285,-0.1066],[286,-0.11],[287,-0.1088],[293,-0.1088],[294,-0.1116],[302,-0.1116],[303,-0.1111],[304,-0.1115],[309,-0.1115],[323,-0.1053],[324,-0.108],[326,-0.108],[327,-0.1019],[331,-0.1019],[332,-0.0975],[337,-0.0975],[338,-0.085],[339,-0.085],[340,-0.0872],[341,-0.0872],[342,-0.0728],[345,-0.0728],[346,-0.0591],[349,-0.0591],[350,-0.0455],[354,-0.0455],[355,-0.0342],[356,-0.0346],[361,-0.0346],[362,-0.0386],[366,-0.0386],[367,-0.0521],[370,-0.0521],[371,-0.0677],[372,-0.0661],[374,-0.0661],[375,-0.0831],[378,-0.0831],[379,-0.0952],[387,-0.0952],[388,-0.0922],[389,-0.1064],[403,-0.1063],[404,-0.1061],[410,-0.106],[411,-0.11],[412,-0.11],[413,-0.1303],[414,-0.144],[415,-0.1602],[416,-0.1805],[417,-0.1959],[418,-0.2105],[419,-0.2276],[420,-0.2264],[421,-0.242],[422,-0.242],[423,-0.2462],[428,-0.2462],[429,-0.2331],[431,-0.2331],[432,-0.221],[433,-0.2187],[434,0.4645],[435,0.4536],[436,0.448],[437,0.4471],[439,0.4471],[440,0.4661],[441,0.4787],[442,0.4942],[443,0.5123],[444,0.5316],[445,-0.1528],[446,-0.1473],[448,-0.1473],[449,-0.1326],[451,-0.1326],[452,-0.1199],[455,-0.1199],[456,-0.1071],[459,-0.1071],[460,-0.1041],[467,-0.104],[468,-0.1038],[486,-0.1037],[487,-0.1032],[489,-0.1032],[490,-0.103],[495,-0.1031],[496,-0.1026],[499,-0.1026],[500,-0.1014],[515,-0.1014],[530,-0.1014],[531,-0.101],[534,-0.1011],[535,-0.1009],[542,-0.101],[543,-0.1013],[547,-0.1013],[548,-0.1002],[551,-0.1002],[552,-0.1043],[553,-0.1043],[554,-0.1062],[555,-0.1062],[556,-0.1141],[557,-0.1141],[558,-0.1221],[559,-0.1221],[560,-0.1279],[561,-0.1279],[562,-0.1337],[563,-0.1337],[564,-0.1401],[565,-0.1401],[566,-0.1445],[572,-0.1445],[573,-0.14],[574,-0.14],[576,-0.1275],[577,-0.1222],[578,-0.1222],[579,-0.1116],[580,-0.1091],[581,-0.1091],[582,-0.1004],[583,-0.1004],[584,-0.095],[585,-0.095],[586,-0.0915],[591,-0.0915],[592,-0.0898],[594,-0.0898],[595,-0.0945],[597,-0.0946],[598,-0.098],[600,-0.098],[601,-0.0997],[604,-0.0997],[605,-0.1066],[610,-0.1066],[611,-0.11],[612,-0.1088],[618,-0.1088],[619,-0.1116],[627,-0.1116],[628,-0.1111],[629,-0.1115],[634,-0.1115],[676,-0.1053],[677,-0.108],[679,-0.108],[680,-0.1019],[684,-0.1019],[685,-0.0975],[690,-0.0975],[691,-0.085],[692,-0.085],[693,-0.0872],[694,-0.0872],[695,-0.0728],[698,-0.0728],[699,-0.0591],[702,-0.0591],[703,-0.0455],[707,-0.0455],[708,-0.0342],[709,-0.0346],[714,-0.0346],[715,-0.0386],[719,-0.0386],[720,-0.0521],[723,-0.0521],[724,-0.0677],[725,-0.0661],[727,-0.0661],[728,-0.0831],[731,-0.0831],[732,-0.0952],[740,-0.0952],[741,-0.0922],[742,-0.1064],[755,-0.1063],[757,-0.1061],[770,-0.1066],[771,-0.11],[772,-0.1088],[778,-0.1088],[779,-0.1116],[787,-0.1116],[788,-0.1111],[789,-0.1115],[795,-0.1115],[796,-0.11],[803,-0.11],[837,-0.1053],[887,-0.1115],[888,-0.11],[895,-0.11],[929,-0.1053]]
       }
       ball_l_fk_ctrl.Rotation.Z {
        type=MovieSceneScriptingFloatChannel
        default=0
        thinned_from=751
        keys [[-2,0.6363],[-1,0.6667],[1,0.6667],[2,0.6769],[6,0.6769],[7,0.687],[12,0.687],[13,0.6973],[16,0.697],[17,0.7009],[20,0.7009],[21,0.6972],[24,0.6972],[25,0.7012],[29,0.7012],[30,0.7107],[36,0.7112],[37,0.7244],[41,0.7244],[42,0.7342],[45,0.7342],[46,0.744],[49,0.743],[50,0.7469],[53,0.7469],[54,0.7489],[63,0.7489],[64,0.7447],[72,0.7447],[73,0.7285],[78,0.7285],[81,0.7299],[82,0.7139],[85,0.7139],[86,0.6978],[87,0.6978],[88,0.688],[91,0.6685],[92,0.6588],[94,0.6458],[95,0.6444],[96,0.631],[97,0.631],[98,0.6243],[103,0.6243],[104,0.6177],[106,0.6177],[108,0.6577],[109,-0.973],[110,-1.0058],[111,-1.0322],[112,-1.0474],[114,-1.0474],[115,-1.0406],[116,-1.0221],[117,-1.0002],[118,-0.9698],[119,-0.9344],[120,0.6788],[121,0.6617],[123,0.6617],[124,0.6499],[126,0.6499],[127,0.639],[130,0.639],[131,0.6278],[134,0.6278],[135,0.6127],[138,0.6127],[139,0.6013],[142,0.6014],[143,0.5862],[146,0.5862],[147,0.5749],[151,0.5749],[152,0.556],[156,0.556],[157,0.5447],[158,0.5447],[159,0.5428],[161,0.5428],[162,0.53],[164,0.53],[165,0.5173],[167,0.5173],[168,0.5045],[170,0.5045],[171,0.4918],[174,0.4918],[175,0.4787],[190,0.4787],[191,0.4815],[193,0.4815],[194,0.4896],[200,0.4896],[201,0.5057],[205,0.5057],[206,0.5219],[207,0.5159],[209,0.5159],[210,0.5321],[213,0.5321],[214,0.5483],[217,0.5483],[218,0.5563],[222,0.5563],[223,0.5735],[226,0.5735],[227,0.5895],[228,0.5895],[229,0.6055],[230,0.6055],[231,0.6295],[232,0.6295],[233,0.6455],[234,0.6455],[235,0.6614],[236,0.6614],[237,0.6774],[238,0.6774],[239,0.6829],[240,0.6829],[241,0.6856],[247,0.6856],[248,0.6733],[249,0.6733],[250,0.6651],[252,0.6598],[253,0.6598],[254,0.6531],[255,0.6495],[256,0.6495],[257,0.6416],[258,0.6416],[259,0.6337],[260,0.6337],[261,0.6306],[266,0.6306],[267,0.6146],[269,0.6146],[270,0.6097],[272,0.6101],[273,0.604],[275,0.604],[276,0.5978],[279,0.5978],[280,0.5917],[285,0.5917],[286,0.5855],[293,0.5845],[294,0.5783],[302,0.5783],[303,0.5768],[304,0.5604],[309,0.5604],[323,0.6363],[324,0.6667],[326,0.6667],[327,0.6769],[331,0.6769],[332,0.687],[337,0.687],[338,0.6973],[341,0.697],[342,0.7009],[345,0.7009],[346,0.6972],[349,0.6972],[350,0.7012],[354,0.7012],[355,0.7107],[361,0.7112],[362,0.7244],[366,0.7244],[367,0.7342],[370,0.7342],[371,0.744],[374,0.743],[375,0.7469],[378,0.7469],[379,0.7489],[388,0.7489],[389,0.7447],[397,0.7447],[398,0.7285],[403,0.7285],[406,0.7299],[407,0.7139],[410,0.7139],[411,0.6978],[412,0.6978],[413,0.688],[416,0.6685],[417,0.6588],[419,0.6458],[420,0.6444],[421,0.631],[422,0.631],[423,0.6243],[428,0.6243],[429,0.6177],[431,0.6177],[433,0.6577],[434,-0.973],[435,-1.0058],[436,-1.0322],[437,-1.0474],[439,-1.0474],[440,-1.0406],[441,-1.0221],[442,-1.0002],[443,-0.9698],[444,-0.9344],[445,0.6788],[446,0.6617],[448,0.6617],[449,0.6499],[451,0.6499],[452,0.639],[455,0.639],[456,0.6278],[459,0.6278],[460,0.6127],[463,0.6127],[464,0.6013],[467,0.6014],[468,0.5862],[471,0.5862],[472,0.5749],[476,0.5749],[477,0.556],[481,0.556],[482,0.5447],[483,0.5447],[484,0.5428],[486,0.5428],[487,0.53],[489,0.53],[490,0.5173],[492,0.5173],[493,0.5045],[495,0.5045],[496,0.4918],[499,0.4918],[500,0.4787],[515,0.4787],[516,0.4815],[518,0.4815],[519,0.4896],[525,0.4896],[526,0.5057],[530,0.5057],[531,0.5219],[532,0.5159],[534,0.5159],[535,0.5321],[538,0.5321],[539,0.5483],[542,0.5483],[543,0.5563],[547,0.5563],[548,0.5735],[551,0.5735],[552,0.5895],[553,0.5895],[554,0.6055],[555,0.6055],[556,0.6295],[557,0.6295],[558,0.6455],[559,0.6455],[560,0.6614],[561,0.6614],[562,0.6774],[563,0.6774],[564,0.6829],[565,0.6829],[566,0.6856],[572,0.6856],[573,0.6733],[574,0.6733],[575,0.6651],[577,0.6598],[578,0.6598],[579,0.6531],[580,0.6495],[581,0.6495],[582,0.6416],[583,0.6416],[584,0.6337],[585,0.6337],[586,0.6306],[591,0.6306],[592,0.6146],[594,0.6146],[595,0.6097],[597,0.6101],[598,0.604],[600,0.604],[601,0.5978],[604,0.5978],[605,0.5917],[610,0.5917],[611,0.5855],[618,0.5845],[619,0.5783],[627,0.5783],[628,0.5768],[629,0.5604],[634,0.5604],[676,0.6363],[677,0.6667],[679,0.6667],[680,0.6769],[684,0.6769],[685,0.687],[690,0.687],[691,0.6973],[694,0.697],[695,0.7009],[698,0.7009],[699,0.6972],[702,0.6972],[703,0.7012],[707,0.7012],[708,0.7107],[714,0.7112],[715,0.7244],[719,0.7244],[720,0.7342],[723,0.7342],[724,0.744],[727,0.743],[728,0.7469],[731,0.7469],[732,0.7489],[741,0.7489],[742,0.7447],[750,0.7447],[751,0.7285],[755,0.7285],[759,0.7299],[770,0.5917],[771,0.5855],[778,0.5845],[779,0.5783],[787,0.5783],[788,0.5768],[789,0.5604],[795,0.5604],[796,0.5481],[803,0.5481],[837,0.6363],[887,0.5604],[888,0.5481],[895,0.5481],[929,0.6363]]
       }
       ball_l_fk_ctrl.Scale.X {type=MovieSceneScriptingFloatChannel, default=1, const [1,751]}
       ball_l_fk_ctrl.Scale.Y {type=MovieSceneScriptingFloatChannel, default=1, const [1,751]}
       ball_l_fk_ctrl.Scale.Z {type=MovieSceneScriptingFloatChannel, default=1, const [1,751]}
       foot_l_ik_ctrl.Location.X {
        type=MovieSceneScriptingFloatChannel
        default=0
        thinned_from=753
        keys [[-2,-3.5597],[0,-3.5608],[1,-3.6391],[2,-3.8239],[3,-4.0251],[4,-4.1617],[7,-4.3095],[10,-4.4734],[20,-4.9076],[21,-4.9294],[33,-5.3749],[40,-5.5449],[46,-5.6385],[51,-5.6699],[58,-5.6335],[68,-5.454],[83,-5.1031],[85,-5.0282],[88,-4.8945],[91,-4.7889],[99,-4.5572],[101,-4.5278],[102,-4.5356],[103,-4.5549],[104,-4.596],[105,-4.6582],[106,-4.7506],[107,-4.8709],[108,-5.0263],[109,-5.2123],[110,-5.4688],[111,-5.6608],[118,-7.1536],[120,-7.5465],[122,-7.8882],[125,-8.1388],[129,-8.3824],[149,-9.3153],[155,-9.4934],[158,-9.5403],[162,-9.4772],[180,-8.9645],[203,-8.5613],[206,-8.5852],[212,-8.8041],[214,-8.9091],[222,-9.5027],[224,-9.6052],[226,-9.6481],[227,-9.6431],[228,-9.6121],[229,-9.5594],[230,-9.4758],[231,-9.3639],[232,-9.2213],[233,-9.0457],[234,-8.8395],[235,-8.6001],[236,-8.3276],[237,-8.0253],[238,-7.6944],[239,-7.3379],[240,-6.9589],[241,-6.5615],[243,-5.7226],[246,-4.3937],[247,-3.9628],[248,-3.5425],[249,-3.1417],[250,-2.7574],[251,-2.3933],[252,-2.0561],[253,-1.737],[254,-1.4476],[255,-1.1753],[256,-0.9303],[257,-0.721],[258,-0.5548],[259,-0.4385],[260,-0.3689],[261,-0.3335],[262,-0.3413],[263,-0.3831],[264,-0.4452],[265,-0.5317],[266,-0.6317],[267,-0.7512],[268,-0.8783],[272,-1.4263],[273,-1.5542],[274,-1.6663],[275,-1.7721],[276,-1.8495],[277,-1.912],[280,-2.0529],[283,-2.1805],[284,-2.2145],[286,-2.2983],[288,-2.3674],[291,-2.4467],[293,-2.4799],[295,-2.4957],[296,-2.4973],[297,-2.4923],[299,-2.4676],[301,-2.426],[302,-2.3982],[304,-2.317],[305,-2.2671],[307,-2.1526],[309,-2.0195],[323,-3.5597],[325,-3.5608],[326,-3.6391],[327,-3.8239],[328,-4.0251],[329,-4.1617],[332,-4.3095],[335,-4.4734],[345,-4.9076],[346,-4.9294],[358,-5.3749],[365,-5.5449],[371,-5.6385],[376,-5.6699],[383,-5.6335],[393,-5.454],[408,-5.1031],[410,-5.0282],[413,-4.8945],[416,-4.7889],[424,-4.5572],[426,-4.5278],[427,-4.5356],[428,-4.5549],[429,-4.596],[430,-4.6582],[431,-4.7506],[432,-4.8709],[433,-5.0263],[434,-5.2123],[435,-5.4688],[436,-5.6608],[443,-7.1536],[445,-7.5465],[447,-7.8882],[450,-8.1388],[454,-8.3824],[474,-9.3153],[480,-9.4934],[483,-9.5403],[487,-9.4772],[505,-8.9645],[528,-8.5613],[531,-8.5852],[537,-8.8041],[539,-8.9091],[547,-9.5027],[549,-9.6052],[551,-9.6481],[552,-9.6431],[553,-9.6121],[554,-9.5594],[555,-9.4758],[556,-9.3639],[557,-9.2213],[558,-9.0457],[559,-8.8395],[560,-8.6001],[561,-8.3276],[562,-8.0253],[563,-7.6944],[564,-7.3379],[565,-6.9589],[566,-6.5615],[568,-5.7226],[571,-4.3937],[572,-3.9628],[573,-3.5425],[574,-3.1417],[575,-2.7574],[576,-2.3933],[577,-2.0561],[578,-1.737],[579,-1.4476],[580,-1.1753],[581,-0.9303],[582,-0.721],[583,-0.5548],[584,-0.4385],[585,-0.3689],[586,-0.3335],[587,-0.3413],[588,-0.3831],[589,-0.4452],[590,-0.5317],[591,-0.6317],[592,-0.7512],[593,-0.8783],[597,-1.4263],[598,-1.5542],[599,-1.6663],[600,-1.7721],[601,-1.8495],[602,-1.912],[605,-2.0529],[608,-2.1805],[609,-2.2145],[611,-2.2983],[613,-2.3674],[616,-2.4467],[618,-2.4799],[620,-2.4957],[621,-2.4973],[622,-2.4923],[624,-2.4676],[626,-2.426],[627,-2.3982],[629,-2.317],[630,-2.2671],[632,-2.1526],[634,-2.0195],[676,-3.5597],[678,-3.5608],[679,-3.6391],[680,-3.8239],[681,-4.0251],[682,-4.1617],[685,-4.3095],[688,-4.4734],[698,-4.9076],[699,-4.9294],[711,-5.3749],[718,-5.5449],[724,-5.6385],[729,-5.6699],[736,-5.6335],[746,-5.454],[757,-5.1894],[759,-5.1486],[770,-2.2561],[771,-2.2983],[773,-2.3674],[776,-2.4467],[778,-2.4799],[780,-2.4957],[781,-2.4973],[782,-2.4923],[784,-2.4676],[786,-2.426],[787,-2.3982],[789,-2.317],[790,-2.2671],[792,-2.1526],[794,-2.0195],[795,-1.9492],[803,-1.3406],[834,-3.5365],[837,-3.5597],[887,-1.9492],[895,-1.3406],[926,-3.5365],[929,-3.5597]]
       }
       foot_l_ik_ctrl.Location.Y {
        type=MovieSceneScriptingFloatChannel
        default=0
        thinned_from=753
        keys [[-2,-13.3034],[-1,-13.5653],[0,-14.1626],[1,-14.8022],[2,-15.1944],[4,-15.4195],[13,-16.8098],[16,-17.1673],[18,-17.3481],[19,-17.3528],[24,-17.6733],[27,-17.773],[33,-17.6799],[38,-17.4064],[44,-16.8585],[67,-13.7761],[72,-13.2054],[79,-12.5963],[81,-12.446],[94,-11.0596],[98,-10.5438],[99,-10.4298],[100,-10.3455],[101,-10.2858],[102,-10.2538],[103,-10.2737],[104,-10.3546],[105,-10.502],[106,-10.7401],[107,-11.0664],[108,-11.598],[109,-11.8914],[110,-12.2405],[111,-12.6284],[112,-13.0479],[114,-13.9761],[117,-15.4198],[118,-15.8599],[119,-16.2534],[120,-16.5979],[121,-16.6737],[122,-16.8182],[124,-17.0044],[131,-17.3764],[150,-19.2353],[155,-19.4866],[156,-19.5035],[159,-19.3164],[200,-15.1384],[202,-14.9989],[204,-14.9161],[206,-14.8874],[209,-14.9426],[210,-14.9852],[211,-15.0711],[212,-15.2252],[213,-15.4515],[214,-15.7463],[215,-16.1055],[216,-16.5368],[217,-17.0402],[218,-17.6086],[219,-18.2386],[221,-19.6389],[223,-21.1736],[228,-25.1835],[230,-26.6492],[231,-27.3031],[232,-27.8927],[233,-28.3986],[234,-28.823],[235,-29.1526],[236,-29.3804],[237,-29.5008],[238,-29.5027],[239,-29.3782],[240,-29.1282],[241,-28.7378],[242,-28.2062],[243,-27.5581],[246,-25.2884],[252,-20.3792],[260,-14.1379],[261,-13.3747],[263,-11.9433],[265,-10.6119],[266,-9.9822],[267,-9.3916],[268,-8.8203],[269,-8.2667],[271,-7.2338],[272,-6.7481],[273,-6.287],[274,-5.8425],[275,-5.4283],[276,-5.033],[277,-4.6539],[278,-4.301],[279,-3.9637],[280,-3.6433],[281,-3.3385],[282,-3.0446],[283,-2.7685],[284,-2.5033],[285,-2.2542],[286,-2.0147],[287,-1.7932],[288,-1.591],[289,-1.4074],[290,-1.2533],[293,-0.8224],[294,-0.6805],[295,-0.5484],[296,-0.4139],[297,-0.2842],[298,-0.1575],[299,-0.0323],[300,0.086],[301,0.2062],[302,0.3205],[303,0.4304],[304,0.5332],[305,0.638],[306,0.7371],[307,0.8291],[308,0.923],[309,1.0096],[323,-13.3034],[324,-13.5653],[325,-14.1626],[326,-14.8022],[327,-15.1944],[329,-15.4195],[338,-16.8098],[341,-17.1673],[343,-17.3481],[344,-17.3528],[349,-17.6733],[352,-17.773],[358,-17.6799],[363,-17.4064],[369,-16.8585],[392,-13.7761],[397,-13.2054],[404,-12.5963],[406,-12.446],[419,-11.0596],[423,-10.5438],[424,-10.4298],[425,-10.3455],[426,-10.2858],[427,-10.2538],[428,-10.2737],[429,-10.3546],[430,-10.502],[431,-10.7401],[432,-11.0664],[433,-11.598],[434,-11.8914],[435,-12.2405],[436,-12.6284],[437,-13.0479],[439,-13.9761],[442,-15.4198],[443,-15.8599],[444,-16.2534],[445,-16.5979],[446,-16.6737],[447,-16.8182],[449,-17.0044],[456,-17.3764],[475,-19.2353],[480,-19.4866],[481,-19.5035],[484,-19.3164],[525,-15.1384],[527,-14.9989],[529,-14.9161],[531,-14.8874],[534,-14.9426],[535,-14.9852],[536,-15.0711],[537,-15.2252],[538,-15.4515],[539,-15.7463],[540,-16.1055],[541,-16.5368],[542,-17.0402],[543,-17.6086],[544,-18.2386],[546,-19.6389],[548,-21.1736],[553,-25.1835],[555,-26.6492],[556,-27.3031],[557,-27.8927],[558,-28.3986],[559,-28.823],[560,-29.1526],[561,-29.3804],[562,-29.5008],[563,-29.5027],[564,-29.3782],[565,-29.1282],[566,-28.7378],[567,-28.2062],[568,-27.5581],[571,-25.2884],[577,-20.3792],[585,-14.1379],[586,-13.3747],[588,-11.9433],[590,-10.6119],[591,-9.9822],[592,-9.3916],[593,-8.8203],[594,-8.2667],[596,-7.2338],[597,-6.7481],[598,-6.287],[599,-5.8425],[600,-5.4283],[601,-5.033],[602,-4.6539],[603,-4.301],[604,-3.9637],[605,-3.6433],[606,-3.3385],[607,-3.0446],[608,-2.7685],[609,-2.5033],[610,-2.2542],[611,-2.0147],[612,-1.7932],[613,-1.591],[614,-1.4074],[615,-1.2533],[618,-0.8224],[619,-0.6805],[620,-0.5484],[621,-0.4139],[622,-0.2842],[623,-0.1575],[624,-0.0323],[625,0.086],[626,0.2062],[627,0.3205],[628,0.4304],[629,0.5332],[630,0.638],[631,0.7371],[632,0.8291],[633,0.923],[634,1.0096],[676,-13.3034],[677,-13.5653],[678,-14.1626],[679,-14.8022],[680,-15.1944],[682,-15.4195],[691,-16.8098],[694,-17.1673],[696,-17.3481],[697,-17.3528],[702,-17.6733],[705,-17.773],[711,-17.6799],[716,-17.4064],[722,-16.8585],[745,-13.7761],[750,-13.2054],[755,-12.7551],[759,-12.446],[770,-2.2542],[771,-2.0147],[772,-1.7932],[773,-1.591],[774,-1.4074],[775,-1.2533],[778,-0.8224],[779,-0.6805],[780,-0.5484],[781,-0.4139],[782,-0.2842],[783,-0.1575],[784,-0.0323],[785,0.086],[786,0.2062],[787,0.3205],[788,0.4304],[789,0.5332],[790,0.638],[791,0.7371],[792,0.8291],[793,0.923],[794,1.0096],[795,1.0929],[797,1.248],[798,1.3209],[799,1.3904],[801,1.525],[803,1.6555],[834,-12.8946],[837,-13.3034],[887,1.0929],[889,1.248],[890,1.3209],[891,1.3904],[893,1.525],[895,1.6555],[926,-12.8946],[929,-13.3034]]
       }
       foot_l_ik_ctrl.Location.Z {
        type=MovieSceneScriptingFloatChannel
        default=0
        thinned_from=753
        keys [[-2,39.7618],[-1,39.8447],[1,40.2064],[2,40.2869],[6,39.8564],[10,39.0344],[22,35.7847],[26,35.033],[29,34.709],[33,34.692],[61,37.4187],[66,37.6201],[70,37.4598],[75,36.8557],[90,34.1457],[94,33.6727],[97,33.5285],[99,33.5886],[101,33.8043],[103,34.1889],[105,34.7498],[107,35.5045],[108,36.0522],[115,39.0781],[120,41.3758],[128,43.6073],[133,44.5018],[138,44.8751],[142,44.7432],[157,42.6145],[178,39.0998],[195,37.5734],[201,37.3888],[206,37.6378],[212,38.3371],[217,39.3388],[222,40.79],[230,43.5646],[233,44.3469],[235,44.6861],[237,44.8231],[239,44.7241],[241,44.3594],[250,41.6003],[260,39.6441],[265,39.0958],[294,39.0078],[304,38.3171],[309,37.6872],[323,39.7618],[324,39.8447],[326,40.2064],[327,40.2869],[331,39.8564],[335,39.0344],[347,35.7847],[351,35.033],[354,34.709],[358,34.692],[386,37.4187],[391,37.6201],[395,37.4598],[400,36.8557],[415,34.1457],[419,33.6727],[422,33.5285],[424,33.5886],[426,33.8043],[428,34.1889],[430,34.7498],[432,35.5045],[433,36.0522],[440,39.0781],[445,41.3758],[453,43.6073],[458,44.5018],[463,44.8751],[467,44.7432],[482,42.6145],[503,39.0998],[520,37.5734],[526,37.3888],[531,37.6378],[537,38.3371],[542,39.3388],[547,40.79],[555,43.5646],[558,44.3469],[560,44.6861],[562,44.8231],[564,44.7241],[566,44.3594],[575,41.6003],[585,39.6441],[590,39.0958],[619,39.0078],[629,38.3171],[634,37.6872],[677,39.8447],[679,40.2064],[680,40.2869],[684,39.8564],[688,39.0344],[700,35.7847],[704,35.033],[707,34.709],[711,34.692],[739,37.4187],[744,37.6201],[748,37.4598],[753,36.8557],[759,35.7769],[770,39.0949],[784,38.7395],[803,36.3591],[834,39.7209],[837,39.7618],[887,37.5408],[895,36.3591],[926,39.7209],[929,39.7618]]
       }
       foot_l_ik_ctrl.Rotation.X {
        type=MovieSceneScriptingFloatChannel
        default=0
        thinned_from=753
        keys [[-2,41.6729],[-1,42.2293],[0,43.544],[1,45.0593],[2,46.2454],[6,48.286],[11,50.0637],[37,55.8237],[48,57.1949],[57,57.1559],[78,53.998],[82,52.9965],[84,52.2648],[86,51.2966],[88,50.0493],[90,48.5199],[92,46.7013],[94,44.6104],[99,38.6632],[101,36.474],[102,35.4495],[103,34.5358],[104,33.7503],[105,33.1361],[106,32.7205],[107,32.4997],[108,32.451],[109,32.6614],[110,33.0234],[111,33.4902],[113,34.6237],[118,37.9138],[121,39.6304],[122,40.1016],[124,40.7902],[126,41.2294],[130,41.6755],[136,41.7815],[151,41.0317],[156,40.4405],[159,39.7482],[163,38.4325],[171,35.3569],[174,34.3879],[177,33.6596],[180,33.2102],[184,32.9118],[190,32.8877],[200,33.5415],[204,34.0658],[207,34.6331],[210,35.4879],[211,35.895],[212,36.4514],[213,37.1378],[214,37.9415],[215,38.8505],[216,39.8589],[218,42.1482],[223,48.5752],[226,53.1095],[232,63.0381],[234,65.89],[235,67.0841],[236,68.099],[237,68.9259],[238,69.5601],[239,69.9805],[240,70.1845],[241,70.16],[242,69.9113],[244,68.8748],[269,51.9356],[274,49.0532],[278,47.1561],[285,44.6356],[309,39.2218],[323,41.6729],[324,42.2293],[325,43.544],[326,45.0593],[327,46.2454],[331,48.286],[336,50.0637],[362,55.8237],[373,57.1949],[382,57.1559],[403,53.998],[407,52.9965],[409,52.2648],[411,51.2966],[413,50.0493],[415,48.5199],[417,46.7013],[419,44.6104],[424,38.6632],[426,36.474],[427,35.4495],[428,34.5358],[429,33.7503],[430,33.1361],[431,32.7205],[432,32.4997],[433,32.451],[434,32.6614],[435,33.0234],[436,33.4902],[438,34.6237],[443,37.9138],[446,39.6304],[447,40.1016],[449,40.7902],[451,41.2294],[455,41.6755],[461,41.7815],[476,41.0317],[481,40.4405],[484,39.7482],[488,38.4325],[496,35.3569],[499,34.3879],[502,33.6596],[505,33.2102],[509,32.9118],[515,32.8877],[525,33.5415],[529,34.0658],[532,34.6331],[535,35.4879],[536,35.895],[537,36.4514],[538,37.1378],[539,37.9415],[540,38.8505],[541,39.8589],[543,42.1482],[548,48.5752],[551,53.1095],[557,63.0381],[559,65.89],[560,67.0841],[561,68.099],[562,68.9259],[563,69.5601],[564,69.9805],[565,70.1845],[566,70.16],[567,69.9113],[569,68.8748],[594,51.9356],[599,49.0532],[603,47.1561],[610,44.6356],[634,39.2218],[676,41.6729],[677,42.2293],[678,43.544],[679,45.0593],[680,46.2454],[684,48.286],[689,50.0637],[715,55.8237],[726,57.1949],[735,57.1559],[755,54.1957],[759,53.2893],[770,44.6356],[803,37.5171],[834,41.536],[837,41.6729],[887,39.0243],[895,37.5171],[926,41.536],[929,41.6729]]
       }
       foot_l_ik_ctrl.Rotation.Y {
        type=MovieSceneScriptingFloatChannel
        default=0
        thinned_from=753
        keys [[-2,-2.1425],[-1,-2.0748],[0,-1.9136],[1,-1.7201],[2,-1.5644],[3,-1.4861],[4,-1.3935],[5,-1.2665],[6,-1.1148],[7,-0.9522],[8,-0.782],[9,-0.6007],[10,-0.4132],[11,-0.2177],[12,-0.0202],[13,0.196],[14,0.4361],[15,0.6927],[16,0.9686],[17,1.256],[19,1.8434],[20,2.1445],[22,2.7194],[23,2.9931],[24,3.2482],[25,3.4834],[26,3.6888],[27,3.8741],[28,4.0139],[29,4.123],[30,4.1933],[31,4.2153],[32,4.2028],[33,4.145],[34,4.0472],[35,3.9136],[36,3.7405],[37,3.5397],[38,3.3086],[39,3.0526],[40,2.7751],[41,2.4841],[42,2.1756],[43,1.8619],[46,0.8879],[47,0.5653],[48,0.2469],[49,-0.0592],[50,-0.3414],[51,-0.6155],[52,-0.8712],[53,-1.1004],[54,-1.2926],[55,-1.4714],[56,-1.6369],[57,-1.7868],[58,-1.9255],[59,-2.0539],[61,-2.2745],[62,-2.3624],[63,-2.4427],[64,-2.5115],[65,-2.5728],[66,-2.6226],[67,-2.6624],[68,-2.6914],[69,-2.7126],[70,-2.7245],[72,-2.7268],[74,-2.7117],[76,-2.6836],[79,-2.6194],[80,-2.5886],[81,-2.5516],[82,-2.5084],[83,-2.4542],[84,-2.3921],[85,-2.3158],[86,-2.412],[87,-2.8308],[88,-3.5083],[89,-4.3837],[90,-5.4044],[91,-6.4993],[92,-7.6128],[93,-8.6787],[94,-9.6291],[95,-10.4012],[96,-10.9227],[97,-11.2193],[98,-11.3428],[99,-11.2817],[100,-11.0996],[101,-10.8937],[103,-10.4157],[105,-9.8894],[108,-9.0422],[110,-8.4075],[112,-7.7287],[115,-6.6259],[119,-5.0803],[120,-4.6868],[122,-3.8813],[124,-3.1066],[125,-2.7453],[126,-2.4022],[127,-2.0785],[128,-1.7797],[129,-1.5064],[130,-1.2612],[131,-1.0556],[132,-0.8791],[133,-0.7398],[134,-0.6445],[135,-0.5885],[136,-0.556],[137,-0.5201],[138,-0.4854],[139,-0.4535],[140,-0.4139],[141,-0.3791],[142,-0.3432],[143,-0.3065],[144,-0.2731],[145,-0.2377],[147,-0.1704],[148,-0.1364],[149,-0.1017],[150,-0.0669],[151,-0.0351],[152,-0.0021],[153,0.0286],[154,0.0605],[155,0.0904],[156,0.12],[157,0.1473],[158,0.1738],[159,0.2006],[160,0.2299],[161,0.2601],[162,0.2876],[163,0.3192],[164,0.3482],[166,0.4104],[168,0.4681],[169,0.4983],[171,0.5515],[172,0.577],[173,0.5992],[175,0.6402],[176,0.6536],[177,0.6685],[178,0.678],[179,0.6828],[180,0.6861],[181,0.6871],[183,0.6813],[185,0.6708],[187,0.6533],[188,0.6429],[190,0.6172],[191,0.6033],[194,0.5577],[196,0.522],[197,0.5065],[198,0.4888],[199,0.4728],[201,0.4375],[202,0.413],[203,0.3794],[204,0.3364],[205,0.2903],[206,0.2391],[207,0.1855],[208,0.1307],[209,0.0807],[210,0.0323],[211,-0.0045],[212,-0.03],[213,-0.0317],[214,-0.0034],[215,0.0539],[216,0.1405],[217,0.2495],[218,0.3832],[219,0.5365],[220,0.7134],[221,0.908],[222,1.1212],[223,1.3497],[224,1.5884],[225,1.79],[226,1.9148],[227,1.9626],[228,1.9403],[229,1.8489],[230,1.6997],[231,1.4966],[232,1.2516],[233,0.9712],[234,0.6665],[235,0.3466],[236,0.0227],[237,-0.2961],[238,-0.6025],[239,-0.8784],[240,-1.1193],[241,-1.3019],[242,-1.4067],[243,-1.43],[244,-1.3523],[245,-1.1428],[246,-0.7862],[247,-0.3121],[248,0.27],[249,0.9416],[250,1.6789],[251,2.4587],[253,4.0813],[254,4.8826],[255,5.6241],[256,6.2698],[257,6.805],[258,7.2248],[259,7.5154],[260,7.6706],[261,7.6811],[262,7.5459],[263,7.2554],[264,6.8565],[265,6.3933],[266,5.8832],[267,5.3379],[270,3.6203],[271,3.0591],[272,2.5268],[273,2.0274],[274,1.5821],[275,1.1938],[276,0.8527],[277,0.5415],[278,0.2659],[279,0.0318],[280,-0.1784],[281,-0.3771],[282,-0.5661],[283,-0.7459],[284,-0.9126],[285,-1.072],[286,-1.2206],[287,-1.3569],[288,-1.486],[289,-1.6024],[290,-1.709],[291,-1.8027],[292,-1.8793],[293,-1.9413],[294,-1.9858],[295,-2.0097],[296,-2.0182],[297,-2.0061],[298,-1.9797],[299,-1.9468],[300,-1.901],[301,-1.8438],[302,-1.7754],[303,-1.6966],[304,-1.6054],[305,-1.5076],[306,-1.4023],[307,-1.2874],[308,-1.1687],[309,-1.0428],[323,-2.1425],[324,-2.0748],[325,-1.9136],[326,-1.7201],[327,-1.5644],[328,-1.4861],[329,-1.3935],[330,-1.2665],[331,-1.1148],[332,-0.9522],[333,-0.782],[334,-0.6007],[335,-0.4132],[336,-0.2177],[337,-0.0202],[338,0.196],[339,0.4361],[340,0.6927],[341,0.9686],[342,1.256],[344,1.8434],[345,2.1445],[347,2.7194],[348,2.9931],[349,3.2482],[350,3.4834],[351,3.6888],[352,3.8741],[353,4.0139],[354,4.123],[355,4.1933],[356,4.2153],[357,4.2028],[358,4.145],[359,4.0472],[360,3.9136],[361,3.7405],[362,3.5397],[363,3.3086],[364,3.0526],[365,2.7751],[366,2.4841],[367,2.1756],[368,1.8619],[371,0.8879],[372,0.5653],[373,0.2469],[374,-0.0592],[375,-0.3414],[376,-0.6155],[377,-0.8712],[378,-1.1004],[379,-1.2926],[380,-1.4714],[381,-1.6369],[382,-1.7868],[383,-1.9255],[384,-2.0539],[386,-2.2745],[387,-2.3624],[388,-2.4427],[389,-2.5115],[390,-2.5728],[391,-2.6226],[392,-2.6624],[393,-2.6914],[394,-2.7126],[395,-2.7245],[397,-2.7268],[399,-2.7117],[401,-2.6836],[404,-2.6194],[405,-2.5886],[406,-2.5516],[407,-2.5084],[408,-2.4542],[409,-2.3921],[410,-2.3158],[411,-2.412],[412,-2.8308],[413,-3.5083],[414,-4.3837],[415,-5.4044],[416,-6.4993],[417,-7.6128],[418,-8.6787],[419,-9.6291],[420,-10.4012],[421,-10.9227],[422,-11.2193],[423,-11.3428],[424,-11.2817],[425,-11.0996],[426,-10.8937],[428,-10.4157],[430,-9.8894],[433,-9.0422],[435,-8.4075],[437,-7.7287],[440,-6.6259],[444,-5.0803],[445,-4.6868],[447,-3.8813],[449,-3.1066],[450,-2.7453],[451,-2.4022],[452,-2.0785],[453,-1.7797],[454,-1.5064],[455,-1.2612],[456,-1.0556],[457,-0.8791],[458,-0.7398],[459,-0.6445],[460,-0.5885],[461,-0.556],[462,-0.5201],[463,-0.4854],[464,-0.4535],[465,-0.4139],[466,-0.3791],[467,-0.3432],[468,-0.3065],[469,-0.2731],[470,-0.2377],[472,-0.1704],[473,-0.1364],[474,-0.1017],[475,-0.0669],[476,-0.0351],[477,-0.0021],[478,0.0286],[479,0.0605],[480,0.0904],[481,0.12],[482,0.1473],[483,0.1738],[484,0.2006],[485,0.2299],[486,0.2601],[487,0.2876],[488,0.3192],[489,0.3482],[491,0.4104],[493,0.4681],[494,0.4983],[496,0.5515],[497,0.577],[498,0.5992],[500,0.6402],[501,0.6536],[502,0.6685],[503,0.678],[504,0.6828],[505,0.6861],[506,0.6871],[508,0.6813],[510,0.6708],[512,0.6533],[513,0.6429],[515,0.6172],[516,0.6033],[519,0.5577],[521,0.522],[522,0.5065],[523,0.4888],[524,0.4728],[526,0.4375],[527,0.413],[528,0.3794],[529,0.3364],[530,0.2903],[531,0.2391],[532,0.1855],[533,0.1307],[534,0.0807],[535,0.0323],[536,-0.0045],[537,-0.03],[538,-0.0317],[539,-0.0034],[540,0.0539],[541,0.1405],[542,0.2495],[543,0.3832],[544,0.5365],[545,0.7134],[546,0.908],[547,1.1212],[548,1.3497],[549,1.5884],[550,1.79],[551,1.9148],[552,1.9626],[553,1.9403],[554,1.8489],[555,1.6997],[556,1.4966],[557,1.2516],[558,0.9712],[559,0.6665],[560,0.3466],[561,0.0227],[562,-0.2961],[563,-0.6025],[564,-0.8784],[565,-1.1193],[566,-1.3019],[567,-1.4067],[568,-1.43],[569,-1.3523],[570,-1.1428],[571,-0.7862],[572,-0.3121],[573,0.27],[574,0.9416],[575,1.6789],[576,2.4587],[578,4.0813],[579,4.8826],[580,5.6241],[581,6.2698],[582,6.805],[583,7.2248],[584,7.5154],[585,7.6706],[586,7.6811],[587,7.5459],[588,7.2554],[589,6.8565],[590,6.3933],[591,5.8832],[592,5.3379],[595,3.6203],[596,3.0591],[597,2.5268],[598,2.0274],[599,1.5821],[600,1.1938],[601,0.8527],[602,0.5415],[603,0.2659],[604,0.0318],[605,-0.1784],[606,-0.3771],[607,-0.5661],[608,-0.7459],[609,-0.9126],[610,-1.072],[611,-1.2206],[612,-1.3569],[613,-1.486],[614,-1.6024],[615,-1.709],[616,-1.8027],[617,-1.8793],[618,-1.9413],[619,-1.9858],[620,-2.0097],[621,-2.0182],[622,-2.0061],[623,-1.9797],[624,-1.9468],[625,-1.901],[626,-1.8438],[627,-1.7754],[628,-1.6966],[629,-1.6054],[630,-1.5076],[631,-1.4023],[632,-1.2874],[633,-1.1687],[634,-1.0428],[676,-2.1425],[677,-2.0748],[678,-1.9136],[679,-1.7201],[680,-1.5644],[681,-1.4861],[682,-1.3935],[683,-1.2665],[684,-1.1148],[685,-0.9522],[686,-0.782],[687,-0.6007],[688,-0.4132],[689,-0.2177],[690,-0.0202],[691,0.196],[692,0.4361],[693,0.6927],[694,0.9686],[695,1.256],[697,1.8434],[698,2.1445],[700,2.7194],[701,2.9931],[702,3.2482],[703,3.4834],[704,3.6888],[705,3.8741],[706,4.0139],[707,4.123],[708,4.1933],[709,4.2153],[710,4.2028],[711,4.145],[712,4.0472],[713,3.9136],[714,3.7405],[715,3.5397],[716,3.3086],[717,3.0526],[718,2.7751],[719,2.4841],[720,2.1756],[721,1.8619],[724,0.8879],[725,0.5653],[726,0.2469],[727,-0.0592],[728,-0.3414],[729,-0.6155],[730,-0.8712],[731,-1.1004],[732,-1.2926],[733,-1.4714],[734,-1.6369],[735,-1.7868],[736,-1.9255],[737,-2.0539],[739,-2.2745],[740,-2.3624],[741,-2.4427],[742,-2.5115],[743,-2.5728],[744,-2.6226],[745,-2.6624],[746,-2.6914],[747,-2.7126],[748,-2.7245],[750,-2.7268],[752,-2.7117],[754,-2.6836],[755,-2.6653],[757,-2.6194],[759,-2.5516],[770,-1.072],[771,-1.2206],[772,-1.3569],[773,-1.486],[774,-1.6024],[775,-1.709],[776,-1.8027],[777,-1.8793],[778,-1.9413],[779,-1.9858],[780,-2.0097],[781,-2.0182],[782,-2.0061],[783,-1.9797],[784,-1.9468],[785,-1.901],[786,-1.8438],[787,-1.7754],[788,-1.6966],[789,-1.6054],[790,-1.5076],[791,-1.4023],[792,-1.2874],[793,-1.1687],[794,-1.0428],[795,-0.9138],[796,-0.7819],[797,-0.647],[798,-0.5141],[799,-0.3796],[800,-0.248],[801,-0.1207],[802,0.0045],[803,0.1262],[834,-2.1941],[837,-2.1425],[887,-0.9138],[888,-0.7819],[889,-0.647],[890,-0.5141],[891,-0.3796],[892,-0.248],[893,-0.1207],[894,0.0045],[895,0.1262],[926,-2.1941],[929,-2.1425]]
       }
       foot_l_ik_ctrl.Rotation.Z {
        type=MovieSceneScriptingFloatChannel
        default=0
        thinned_from=753
        keys [[-2,-45.2287],[-1,-45.7069],[0,-46.8258],[1,-48.1042],[2,-49.0643],[6,-50.5608],[9,-51.2834],[26,-51.7239],[78,-49.7437],[85,-50.1567],[90,-51.0909],[92,-51.2048],[93,-51.1086],[94,-50.8847],[95,-50.5309],[97,-49.5043],[98,-49.086],[99,-48.8284],[102,-48.4761],[104,-48.4571],[106,-48.692],[108,-49.3026],[112,-51.4554],[114,-52.2986],[125,-55.1968],[132,-56.2114],[215,-56.0451],[218,-56.1639],[220,-55.993],[222,-55.5833],[224,-54.9556],[226,-53.9902],[228,-52.6293],[230,-50.9671],[235,-48.7322],[240,-45.9443],[246,-42.1868],[248,-41.1213],[250,-40.2518],[252,-39.6034],[262,-37.6018],[268,-35.8656],[275,-33.5339],[278,-32.7289],[285,-31.4169],[290,-30.787],[293,-30.5884],[297,-30.5981],[309,-31.5369],[323,-45.2287],[324,-45.7069],[325,-46.8258],[326,-48.1042],[327,-49.0643],[331,-50.5608],[334,-51.2834],[351,-51.7239],[403,-49.7437],[410,-50.1567],[415,-51.0909],[417,-51.2048],[418,-51.1086],[419,-50.8847],[420,-50.5309],[422,-49.5043],[423,-49.086],[424,-48.8284],[427,-48.4761],[429,-48.4571],[431,-48.692],[433,-49.3026],[437,-51.4554],[439,-52.2986],[450,-55.1968],[457,-56.2114],[540,-56.0451],[543,-56.1639],[545,-55.993],[547,-55.5833],[549,-54.9556],[551,-53.9902],[553,-52.6293],[555,-50.9671],[560,-48.7322],[565,-45.9443],[571,-42.1868],[573,-41.1213],[575,-40.2518],[577,-39.6034],[587,-37.6018],[593,-35.8656],[600,-33.5339],[603,-32.7289],[610,-31.4169],[615,-30.787],[618,-30.5884],[622,-30.5981],[634,-31.5369],[676,-45.2287],[677,-45.7069],[678,-46.8258],[679,-48.1042],[680,-49.0643],[684,-50.5608],[687,-51.2834],[704,-51.7239],[755,-49.7324],[759,-49.8575],[770,-31.4169],[775,-30.787],[778,-30.5884],[782,-30.5981],[803,-32.3251],[834,-44.9048],[837,-45.2287],[887,-31.6206],[895,-32.3251],[926,-44.9048],[929,-45.2287]]
       }
       foot_l_ik_ctrl.Scale.X {type=MovieSceneScriptingFloatChannel, default=1, const [1,753]}
       foot_l_ik_ctrl.Scale.Y {type=MovieSceneScriptingFloatChannel, default=1, const [1,753]}
       foot_l_ik_ctrl.Scale.Z {type=MovieSceneScriptingFloatChannel, default=1, const [1,753]}
       ball_l_ik_ctrl.Location.X {type=MovieSceneScriptingFloatChannel, default=0, thinned_from=751, keys [[-2,0],[929,0]]}
       ball_l_ik_ctrl.Location.Y {type=MovieSceneScriptingFloatChannel, default=0, thinned_from=751, keys [[-2,0],[929,0]]}
       ball_l_ik_ctrl.Location.Z {type=MovieSceneScriptingFloatChannel, default=0, thinned_from=751, keys [[-2,0],[929,0]]}
       ball_l_ik_ctrl.Rotation.X {
        type=MovieSceneScriptingFloatChannel
        default=0
        thinned_from=751
        keys [[-2,0.6252],[-1,0.6553],[1,0.6553],[2,0.666],[6,0.666],[7,0.6765],[12,0.6765],[13,0.6877],[16,0.6873],[17,0.6924],[20,0.6924],[21,0.6899],[24,0.6899],[25,0.695],[29,0.695],[30,0.7055],[36,0.7059],[37,0.7187],[41,0.7187],[42,0.7274],[45,0.7273],[46,0.7358],[49,0.735],[50,0.7374],[63,0.7386],[64,0.7332],[72,0.7332],[73,0.717],[78,0.717],[81,0.7184],[82,0.7024],[85,0.7024],[86,0.6861],[87,0.6861],[88,0.6746],[91,0.6508],[92,0.6397],[94,0.6239],[95,0.6226],[96,0.6078],[97,0.6078],[98,0.6007],[103,0.6007],[104,0.5953],[106,0.5953],[108,0.6365],[109,-0.9222],[110,-0.9558],[111,-0.9825],[112,-0.9976],[114,-0.9976],[115,-0.9891],[116,-0.9695],[117,-0.9463],[118,-0.9146],[119,-0.8777],[120,0.6634],[121,0.6469],[123,0.6469],[124,0.6364],[126,0.6364],[127,0.6267],[130,0.6267],[131,0.6166],[134,0.6166],[135,0.6018],[138,0.6018],[139,0.5905],[142,0.5905],[143,0.5755],[146,0.5755],[147,0.5642],[151,0.5642],[152,0.5454],[156,0.5454],[157,0.5341],[158,0.5341],[159,0.5322],[161,0.5322],[162,0.5195],[164,0.5195],[165,0.5068],[167,0.5068],[168,0.4941],[170,0.4941],[171,0.4815],[174,0.4815],[175,0.4685],[190,0.4685],[191,0.4713],[193,0.4713],[194,0.4794],[200,0.4794],[201,0.4955],[205,0.4955],[206,0.5117],[207,0.5056],[209,0.5056],[210,0.5218],[213,0.5218],[214,0.5379],[217,0.5379],[218,0.5459],[222,0.5459],[223,0.5631],[226,0.5631],[227,0.5787],[228,0.5787],[229,0.5944],[230,0.5944],[231,0.6176],[232,0.6176],[233,0.6328],[234,0.6328],[235,0.6481],[236,0.6481],[237,0.6635],[238,0.6635],[239,0.6683],[240,0.6683],[241,0.6706],[247,0.6706],[248,0.6587],[249,0.6587],[250,0.6512],[252,0.647],[253,0.647],[254,0.6413],[255,0.638],[256,0.638],[257,0.631],[258,0.631],[259,0.6236],[260,0.6236],[261,0.6208],[266,0.6208],[267,0.605],[269,0.605],[270,0.5997],[272,0.6001],[273,0.5936],[275,0.5936],[276,0.5874],[279,0.5874],[280,0.5806],[285,0.5806],[286,0.5741],[293,0.5732],[294,0.5668],[302,0.5668],[303,0.5653],[304,0.5489],[309,0.5489],[323,0.6252],[324,0.6553],[326,0.6553],[327,0.666],[331,0.666],[332,0.6765],[337,0.6765],[338,0.6877],[341,0.6873],[342,0.6924],[345,0.6924],[346,0.6899],[349,0.6899],[350,0.695],[354,0.695],[355,0.7055],[361,0.7059],[362,0.7187],[366,0.7187],[367,0.7274],[370,0.7273],[371,0.7358],[374,0.735],[375,0.7374],[388,0.7386],[389,0.7332],[397,0.7332],[398,0.717],[403,0.717],[406,0.7184],[407,0.7024],[410,0.7024],[411,0.6861],[412,0.6861],[413,0.6746],[416,0.6508],[417,0.6397],[419,0.6239],[420,0.6226],[421,0.6078],[422,0.6078],[423,0.6007],[428,0.6007],[429,0.5953],[431,0.5953],[433,0.6365],[434,-0.9222],[435,-0.9558],[436,-0.9825],[437,-0.9976],[439,-0.9976],[440,-0.9891],[441,-0.9695],[442,-0.9463],[443,-0.9146],[444,-0.8777],[445,0.6634],[446,0.6469],[448,0.6469],[449,0.6364],[451,0.6364],[452,0.6267],[455,0.6267],[456,0.6166],[459,0.6166],[460,0.6018],[463,0.6018],[464,0.5905],[467,0.5905],[468,0.5755],[471,0.5755],[472,0.5642],[476,0.5642],[477,0.5454],[481,0.5454],[482,0.5341],[483,0.5341],[484,0.5322],[486,0.5322],[487,0.5195],[489,0.5195],[490,0.5068],[492,0.5068],[493,0.4941],[495,0.4941],[496,0.4815],[499,0.4815],[500,0.4685],[515,0.4685],[516,0.4713],[518,0.4713],[519,0.4794],[525,0.4794],[526,0.4955],[530,0.4955],[531,0.5117],[532,0.5056],[534,0.5056],[535,0.5218],[538,0.5218],[539,0.5379],[542,0.5379],[543,0.5459],[547,0.5459],[548,0.5631],[551,0.5631],[552,0.5787],[553,0.5787],[554,0.5944],[555,0.5944],[556,0.6176],[557,0.6176],[558,0.6328],[559,0.6328],[560,0.6481],[561,0.6481],[562,0.6635],[563,0.6635],[564,0.6683],[565,0.6683],[566,0.6706],[572,0.6706],[573,0.6587],[574,0.6587],[575,0.6512],[577,0.647],[578,0.647],[579,0.6413],[580,0.638],[581,0.638],[582,0.631],[583,0.631],[584,0.6236],[585,0.6236],[586,0.6208],[591,0.6208],[592,0.605],[594,0.605],[595,0.5997],[597,0.6001],[598,0.5936],[600,0.5936],[601,0.5874],[604,0.5874],[605,0.5806],[610,0.5806],[611,0.5741],[618,0.5732],[619,0.5668],[627,0.5668],[628,0.5653],[629,0.5489],[634,0.5489],[676,0.6252],[677,0.6553],[679,0.6553],[680,0.666],[684,0.666],[685,0.6765],[690,0.6765],[691,0.6877],[694,0.6873],[695,0.6924],[698,0.6924],[699,0.6899],[702,0.6899],[703,0.695],[707,0.695],[708,0.7055],[714,0.7059],[715,0.7187],[719,0.7187],[720,0.7274],[723,0.7273],[724,0.7358],[727,0.735],[728,0.7374],[741,0.7386],[742,0.7332],[750,0.7332],[751,0.717],[755,0.717],[759,0.7184],[770,0.5806],[771,0.5741],[778,0.5732],[779,0.5668],[787,0.5668],[788,0.5653],[789,0.5489],[795,0.5489],[796,0.5368],[803,0.5368],[837,0.6252],[887,0.5489],[888,0.5368],[895,0.5368],[929,0.6252]]
       }
       ball_l_ik_ctrl.Rotation.Y {
        type=MovieSceneScriptingFloatChannel
        default=0
        thinned_from=751
        keys [[-2,-0.1578],[-1,-0.1629],[1,-0.1629],[2,-0.1577],[6,-0.1577],[7,-0.1541],[12,-0.1541],[13,-0.1424],[14,-0.1424],[15,-0.1445],[16,-0.1445],[17,-0.1303],[20,-0.1303],[21,-0.1162],[24,-0.1162],[25,-0.1028],[29,-0.1028],[30,-0.0921],[31,-0.0925],[36,-0.0925],[37,-0.0977],[41,-0.0977],[42,-0.1123],[45,-0.1123],[46,-0.1288],[47,-0.1272],[49,-0.1272],[50,-0.1446],[53,-0.1447],[54,-0.157],[62,-0.157],[63,-0.1541],[64,-0.1681],[72,-0.1681],[73,-0.1666],[81,-0.1665],[82,-0.1651],[85,-0.1651],[86,-0.1677],[87,-0.1677],[88,-0.1873],[89,-0.2006],[90,-0.2164],[91,-0.2363],[93,-0.2652],[94,-0.282],[95,-0.2806],[96,-0.2955],[97,-0.2955],[98,-0.2994],[103,-0.2994],[104,-0.2857],[106,-0.2857],[107,-0.2751],[108,-0.2743],[109,0.5529],[110,0.5456],[111,0.5428],[114,0.5434],[115,0.5615],[116,0.572],[117,0.5851],[119,0.6153],[120,-0.2093],[121,-0.2023],[123,-0.2023],[124,-0.1864],[126,-0.1864],[127,-0.1727],[130,-0.1727],[131,-0.1588],[134,-0.1588],[135,-0.1544],[138,-0.1544],[139,-0.1534],[142,-0.1534],[143,-0.152],[146,-0.152],[147,-0.1509],[151,-0.1509],[152,-0.1494],[156,-0.1494],[157,-0.1484],[161,-0.1482],[162,-0.1467],[164,-0.1467],[165,-0.1454],[167,-0.1454],[168,-0.1443],[170,-0.1443],[171,-0.1428],[174,-0.1428],[175,-0.1405],[193,-0.1406],[194,-0.1413],[200,-0.1413],[201,-0.1427],[205,-0.1427],[206,-0.1437],[207,-0.1433],[209,-0.1433],[210,-0.1445],[213,-0.1445],[214,-0.1459],[217,-0.1459],[218,-0.147],[222,-0.147],[223,-0.1473],[226,-0.1473],[227,-0.1528],[228,-0.1528],[229,-0.1563],[230,-0.1563],[231,-0.1666],[232,-0.1666],[233,-0.1761],[234,-0.1761],[235,-0.1836],[236,-0.1836],[237,-0.1912],[238,-0.1912],[239,-0.1983],[240,-0.1983],[241,-0.2031],[247,-0.2031],[248,-0.1974],[249,-0.1974],[250,-0.1902],[251,-0.1834],[252,-0.1777],[253,-0.1777],[254,-0.166],[255,-0.1629],[256,-0.1629],[257,-0.1532],[258,-0.1532],[259,-0.1468],[260,-0.1468],[261,-0.1428],[266,-0.1428],[267,-0.1398],[269,-0.1398],[270,-0.1444],[272,-0.1445],[273,-0.1477],[275,-0.1477],[276,-0.149],[279,-0.149],[280,-0.1556],[285,-0.1556],[286,-0.1588],[287,-0.1574],[293,-0.1574],[294,-0.1599],[302,-0.1599],[303,-0.1593],[304,-0.1583],[323,-0.1578],[324,-0.1629],[326,-0.1629],[327,-0.1577],[331,-0.1577],[332,-0.1541],[337,-0.1541],[338,-0.1424],[339,-0.1424],[340,-0.1445],[341,-0.1445],[342,-0.1303],[345,-0.1303],[346,-0.1162],[349,-0.1162],[350,-0.1028],[354,-0.1028],[355,-0.0921],[356,-0.0925],[361,-0.0925],[362,-0.0977],[366,-0.0977],[367,-0.1123],[370,-0.1123],[371,-0.1288],[372,-0.1272],[374,-0.1272],[375,-0.1446],[378,-0.1447],[379,-0.157],[387,-0.157],[388,-0.1541],[389,-0.1681],[397,-0.1681],[398,-0.1666],[406,-0.1665],[407,-0.1651],[410,-0.1651],[411,-0.1677],[412,-0.1677],[413,-0.1873],[414,-0.2006],[415,-0.2164],[416,-0.2363],[418,-0.2652],[419,-0.282],[420,-0.2806],[421,-0.2955],[422,-0.2955],[423,-0.2994],[428,-0.2994],[429,-0.2857],[431,-0.2857],[432,-0.2751],[433,-0.2743],[434,0.5529],[435,0.5456],[436,0.5428],[439,0.5434],[440,0.5615],[441,0.572],[442,0.5851],[444,0.6153],[445,-0.2093],[446,-0.2023],[448,-0.2023],[449,-0.1864],[451,-0.1864],[452,-0.1727],[455,-0.1727],[456,-0.1588],[459,-0.1588],[460,-0.1544],[463,-0.1544],[464,-0.1534],[467,-0.1534],[468,-0.152],[471,-0.152],[472,-0.1509],[476,-0.1509],[477,-0.1494],[481,-0.1494],[482,-0.1484],[486,-0.1482],[487,-0.1467],[489,-0.1467],[490,-0.1454],[492,-0.1454],[493,-0.1443],[495,-0.1443],[496,-0.1428],[499,-0.1428],[500,-0.1405],[518,-0.1406],[519,-0.1413],[525,-0.1413],[526,-0.1427],[530,-0.1427],[531,-0.1437],[532,-0.1433],[534,-0.1433],[535,-0.1445],[538,-0.1445],[539,-0.1459],[542,-0.1459],[543,-0.147],[547,-0.147],[548,-0.1473],[551,-0.1473],[552,-0.1528],[553,-0.1528],[554,-0.1563],[555,-0.1563],[556,-0.1666],[557,-0.1666],[558,-0.1761],[559,-0.1761],[560,-0.1836],[561,-0.1836],[562,-0.1912],[563,-0.1912],[564,-0.1983],[565,-0.1983],[566,-0.2031],[572,-0.2031],[573,-0.1974],[574,-0.1974],[575,-0.1902],[576,-0.1834],[577,-0.1777],[578,-0.1777],[579,-0.166],[580,-0.1629],[581,-0.1629],[582,-0.1532],[583,-0.1532],[584,-0.1468],[585,-0.1468],[586,-0.1428],[591,-0.1428],[592,-0.1398],[594,-0.1398],[595,-0.1444],[597,-0.1445],[598,-0.1477],[600,-0.1477],[601,-0.149],[604,-0.149],[605,-0.1556],[610,-0.1556],[611,-0.1588],[612,-0.1574],[618,-0.1574],[619,-0.1599],[627,-0.1599],[628,-0.1593],[629,-0.1583],[676,-0.1578],[677,-0.1629],[679,-0.1629],[680,-0.1577],[684,-0.1577],[685,-0.1541],[690,-0.1541],[691,-0.1424],[692,-0.1424],[693,-0.1445],[694,-0.1445],[695,-0.1303],[698,-0.1303],[699,-0.1162],[702,-0.1162],[703,-0.1028],[707,-0.1028],[708,-0.0921],[709,-0.0925],[714,-0.0925],[715,-0.0977],[719,-0.0977],[720,-0.1123],[723,-0.1123],[724,-0.1288],[725,-0.1272],[727,-0.1272],[728,-0.1446],[731,-0.1447],[732,-0.157],[740,-0.157],[741,-0.1541],[742,-0.1681],[750,-0.1681],[751,-0.1666],[759,-0.1665],[770,-0.1556],[771,-0.1588],[772,-0.1574],[778,-0.1574],[779,-0.1599],[787,-0.1599],[788,-0.1593],[789,-0.1583],[795,-0.1583],[796,-0.1557],[803,-0.1557],[837,-0.1578],[887,-0.1583],[888,-0.1557],[895,-0.1557],[929,-0.1578]]
       }
       ball_l_ik_ctrl.Rotation.Z {
        type=MovieSceneScriptingFloatChannel
        default=0
        thinned_from=751
        keys [[-2,-0.0942],[-1,-0.1021],[1,-0.1021],[2,-0.1012],[6,-0.1012],[7,-0.102],[12,-0.102],[13,-0.1109],[16,-0.1109],[17,-0.1203],[20,-0.1203],[21,-0.1305],[24,-0.1305],[25,-0.1407],[29,-0.1407],[30,-0.1471],[31,-0.1475],[36,-0.1475],[37,-0.1436],[41,-0.1436],[42,-0.1338],[45,-0.1338],[46,-0.1222],[47,-0.1227],[49,-0.1227],[50,-0.1112],[53,-0.1112],[54,-0.1019],[62,-0.1019],[63,-0.103],[64,-0.0942],[85,-0.0943],[86,-0.092],[87,-0.092],[88,-0.0798],[89,-0.0741],[90,-0.0644],[91,-0.0523],[92,-0.0417],[93,-0.0304],[94,-0.015],[95,-0.0165],[96,0.0052],[97,0.0052],[98,0.0216],[103,0.0216],[104,0.0149],[106,0.0149],[107,0.0028],[108,-0.0057],[109,0.8225],[110,0.8449],[111,0.8562],[112,0.8614],[114,0.8614],[115,0.8442],[116,0.829],[117,0.8095],[118,0.7861],[119,0.7585],[120,-0.0649],[121,-0.0673],[123,-0.0673],[124,-0.0788],[126,-0.0788],[127,-0.0882],[130,-0.0882],[131,-0.0977],[134,-0.0977],[135,-0.1014],[142,-0.1014],[143,-0.1004],[161,-0.1004],[162,-0.1],[164,-0.1],[167,-0.1002],[168,-0.1009],[170,-0.1009],[171,-0.1005],[174,-0.1005],[190,-0.1003],[193,-0.1004],[194,-0.1022],[205,-0.1021],[206,-0.1025],[209,-0.1024],[210,-0.1026],[217,-0.1026],[218,-0.1023],[226,-0.1022],[227,-0.0982],[228,-0.0982],[229,-0.0841],[230,-0.0841],[231,-0.0681],[232,-0.0681],[233,-0.0522],[234,-0.0522],[235,-0.0341],[236,-0.0341],[237,-0.0161],[238,-0.0161],[239,-0.0013],[240,-0.0013],[241,0.009],[247,0.009],[248,-0.0014],[249,-0.0014],[250,-0.0174],[251,-0.0296],[252,-0.0427],[253,-0.0427],[254,-0.0691],[255,-0.0821],[256,-0.0821],[257,-0.1033],[258,-0.1033],[259,-0.1228],[260,-0.1228],[261,-0.1343],[266,-0.1343],[267,-0.1327],[269,-0.1327],[270,-0.1201],[271,-0.1213],[272,-0.1213],[273,-0.1039],[275,-0.1039],[276,-0.0952],[279,-0.0952],[280,-0.0813],[285,-0.0813],[286,-0.0639],[287,-0.0669],[293,-0.0669],[294,-0.0536],[302,-0.0536],[303,-0.0541],[304,-0.0538],[309,-0.0538],[323,-0.0942],[324,-0.1021],[326,-0.1021],[327,-0.1012],[331,-0.1012],[332,-0.102],[337,-0.102],[338,-0.1109],[341,-0.1109],[342,-0.1203],[345,-0.1203],[346,-0.1305],[349,-0.1305],[350,-0.1407],[354,-0.1407],[355,-0.1471],[356,-0.1475],[361,-0.1475],[362,-0.1436],[366,-0.1436],[367,-0.1338],[370,-0.1338],[371,-0.1222],[372,-0.1227],[374,-0.1227],[375,-0.1112],[378,-0.1112],[379,-0.1019],[387,-0.1019],[388,-0.103],[389,-0.0942],[410,-0.0943],[411,-0.092],[412,-0.092],[413,-0.0798],[414,-0.0741],[415,-0.0644],[416,-0.0523],[417,-0.0417],[418,-0.0304],[419,-0.015],[420,-0.0165],[421,0.0052],[422,0.0052],[423,0.0216],[428,0.0216],[429,0.0149],[431,0.0149],[432,0.0028],[433,-0.0057],[434,0.8225],[435,0.8449],[436,0.8562],[437,0.8614],[439,0.8614],[440,0.8442],[441,0.829],[442,0.8095],[443,0.7861],[444,0.7585],[445,-0.0649],[446,-0.0673],[448,-0.0673],[449,-0.0788],[451,-0.0788],[452,-0.0882],[455,-0.0882],[456,-0.0977],[459,-0.0977],[460,-0.1014],[467,-0.1014],[468,-0.1004],[486,-0.1004],[487,-0.1],[489,-0.1],[492,-0.1002],[493,-0.1009],[495,-0.1009],[496,-0.1005],[499,-0.1005],[515,-0.1003],[518,-0.1004],[519,-0.1022],[530,-0.1021],[531,-0.1025],[534,-0.1024],[535,-0.1026],[542,-0.1026],[543,-0.1023],[551,-0.1022],[552,-0.0982],[553,-0.0982],[554,-0.0841],[555,-0.0841],[556,-0.0681],[557,-0.0681],[558,-0.0522],[559,-0.0522],[560,-0.0341],[561,-0.0341],[562,-0.0161],[563,-0.0161],[564,-0.0013],[565,-0.0013],[566,0.009],[572,0.009],[573,-0.0014],[574,-0.0014],[575,-0.0174],[576,-0.0296],[577,-0.0427],[578,-0.0427],[579,-0.0691],[580,-0.0821],[581,-0.0821],[582,-0.1033],[583,-0.1033],[584,-0.1228],[585,-0.1228],[586,-0.1343],[591,-0.1343],[592,-0.1327],[594,-0.1327],[595,-0.1201],[596,-0.1213],[597,-0.1213],[598,-0.1039],[600,-0.1039],[601,-0.0952],[604,-0.0952],[605,-0.0813],[610,-0.0813],[611,-0.0639],[612,-0.0669],[618,-0.0669],[619,-0.0536],[627,-0.0536],[628,-0.0541],[629,-0.0538],[634,-0.0538],[676,-0.0942],[677,-0.1021],[679,-0.1021],[680,-0.1012],[684,-0.1012],[685,-0.102],[690,-0.102],[691,-0.1109],[694,-0.1109],[695,-0.1203],[698,-0.1203],[699,-0.1305],[702,-0.1305],[703,-0.1407],[707,-0.1407],[708,-0.1471],[709,-0.1475],[714,-0.1475],[715,-0.1436],[719,-0.1436],[720,-0.1338],[723,-0.1338],[724,-0.1222],[725,-0.1227],[727,-0.1227],[728,-0.1112],[731,-0.1112],[732,-0.1019],[740,-0.1019],[741,-0.103],[742,-0.0942],[759,-0.0943],[770,-0.0813],[771,-0.0639],[772,-0.0669],[778,-0.0669],[779,-0.0536],[787,-0.0536],[788,-0.0541],[789,-0.0538],[795,-0.0538],[796,-0.0552],[803,-0.0552],[837,-0.0942],[887,-0.0538],[888,-0.0552],[895,-0.0552],[929,-0.0942]]
       }
       ball_l_ik_ctrl.Scale.X {type=MovieSceneScriptingFloatChannel, default=1, const [1,751]}
       ball_l_ik_ctrl.Scale.Y {type=MovieSceneScriptingFloatChannel, default=1, const [1,751]}
       ball_l_ik_ctrl.Scale.Z {type=MovieSceneScriptingFloatChannel, default=1, const [1,751]}
       foot_bk1_l_ctrl.Location.X {type=MovieSceneScriptingFloatChannel, default=0, const [0,751]}
       foot_bk1_l_ctrl.Location.Y {type=MovieSceneScriptingFloatChannel, default=0, const [0,751]}
       foot_bk1_l_ctrl.Location.Z {type=MovieSceneScriptingFloatChannel, default=0, const [0,751]}
       foot_bk1_l_ctrl.Rotation.X {type=MovieSceneScriptingFloatChannel, default=0, const [0,751]}
       foot_bk1_l_ctrl.Rotation.Y {type=MovieSceneScriptingFloatChannel, default=0, const [0,751]}
       foot_bk1_l_ctrl.Rotation.Z {type=MovieSceneScriptingFloatChannel, default=0, const [0,751]}
       foot_bk1_l_ctrl.Scale.X {type=MovieSceneScriptingFloatChannel, default=1, const [1,751]}
       foot_bk1_l_ctrl.Scale.Y {type=MovieSceneScriptingFloatChannel, default=1, const [1,751]}
       foot_bk1_l_ctrl.Scale.Z {type=MovieSceneScriptingFloatChannel, default=1, const [1,751]}
       foot_roll_l_ctrl.Location.X {type=MovieSceneScriptingFloatChannel, default=0, const [0,751]}
       foot_roll_l_ctrl.Location.Y {type=MovieSceneScriptingFloatChannel, default=0, const [0,751]}
       foot_roll_l_ctrl.Location.Z {type=MovieSceneScriptingFloatChannel, default=0, const [0,751]}
       foot_roll_l_ctrl.Rotation.X {type=MovieSceneScriptingFloatChannel, default=0, const [0,751]}
       foot_roll_l_ctrl.Rotation.Y {type=MovieSceneScriptingFloatChannel, default=0, const [0,751]}
       foot_roll_l_ctrl.Rotation.Z {type=MovieSceneScriptingFloatChannel, default=0, const [0,751]}
       foot_roll_l_ctrl.Scale.X {type=MovieSceneScriptingFloatChannel, default=1, const [1,751]}
       foot_roll_l_ctrl.Scale.Y {type=MovieSceneScriptingFloatChannel, default=1, const [1,751]}
       foot_roll_l_ctrl.Scale.Z {type=MovieSceneScriptingFloatChannel, default=1, const [1,751]}
       roll_blend_l_ctrl {type=MovieSceneScriptingFloatChannel, default=1, const [1,751]}
       tip_l_ctrl.Location.X {type=MovieSceneScriptingFloatChannel, default=0, const [0,751]}
       tip_l_ctrl.Location.Y {type=MovieSceneScriptingFloatChannel, default=0, const [0,751]}
       tip_l_ctrl.Location.Z {type=MovieSceneScriptingFloatChannel, default=0, const [0,751]}
       tip_l_ctrl.Rotation.X {type=MovieSceneScriptingFloatChannel, default=0, const [0,751]}
       tip_l_ctrl.Rotation.Y {type=MovieSceneScriptingFloatChannel, default=0, const [0,751]}
       tip_l_ctrl.Rotation.Z {type=MovieSceneScriptingFloatChannel, default=0, const [0,751]}
       tip_l_ctrl.Scale.X {type=MovieSceneScriptingFloatChannel, default=1, const [1,751]}
       tip_l_ctrl.Scale.Y {type=MovieSceneScriptingFloatChannel, default=1, const [1,751]}
       tip_l_ctrl.Scale.Z {type=MovieSceneScriptingFloatChannel, default=1, const [1,751]}
       heel_l_ctrl.Location.X {type=MovieSceneScriptingFloatChannel, default=0, const [0,445]}
       heel_l_ctrl.Location.Y {type=MovieSceneScriptingFloatChannel, default=0, const [0,445]}
       heel_l_ctrl.Location.Z {type=MovieSceneScriptingFloatChannel, default=0, const [0,445]}
       heel_l_ctrl.Rotation.X {type=MovieSceneScriptingFloatChannel, default=0, const [0,445]}
       heel_l_ctrl.Rotation.Y {type=MovieSceneScriptingFloatChannel, default=0, const [0,445]}
       heel_l_ctrl.Rotation.Z {type=MovieSceneScriptingFloatChannel, default=0, const [0,445]}
       heel_l_ctrl.Scale.X {type=MovieSceneScriptingFloatChannel, default=1, const [1,445]}
       heel_l_ctrl.Scale.Y {type=MovieSceneScriptingFloatChannel, default=1, const [1,445]}
       heel_l_ctrl.Scale.Z {type=MovieSceneScriptingFloatChannel, default=1, const [1,445]}
       leg_l_softness {type=MovieSceneScriptingFloatChannel, default=0, const [0,753]}
       leg_l_pv_ik_ctrl.Location.X {
        type=MovieSceneScriptingFloatChannel
        default=0
        thinned_from=751
        keys [[-2,-0.3844],[-1,-0.503],[0,-0.7626],[1,-0.8366],[2,-0.9326],[4,-1.0068],[5,-1.0613],[6,-1.1239],[7,-0.9065],[8,-1.2458],[12,-1.4839],[13,-1.5473],[15,-1.6862],[17,-1.8418],[18,-1.9238],[19,-2.0154],[21,-2.1742],[24,-2.3902],[26,-2.5104],[27,-2.5617],[28,-2.6059],[29,-2.6434],[31,-2.6917],[32,-2.7057],[33,-2.7085],[34,-2.705],[35,-2.6959],[37,-2.6594],[40,-2.5825],[43,-2.4871],[46,-2.3845],[49,-2.2902],[52,-2.2083],[55,-2.144],[56,-2.1272],[57,-2.1029],[64,-1.9761],[65,-1.9556],[77,-1.7493],[80,-1.6856],[81,-1.6599],[82,-1.6306],[83,-1.5945],[84,-1.5506],[85,-1.4997],[86,-1.4411],[87,-1.3535],[88,-1.2292],[89,-1.0594],[90,-0.8303],[91,-0.5351],[92,-0.1716],[93,0.2609],[94,0.7669],[95,1.3348],[96,1.9589],[97,2.5573],[98,3.0331],[99,3.3543],[100,3.5699],[101,3.7627],[102,3.9259],[103,4.0515],[104,4.1413],[105,4.1887],[106,4.1903],[107,4.1501],[108,4.0955],[109,3.9337],[110,3.7434],[111,3.5272],[112,3.301],[113,3.0561],[115,2.5387],[118,1.7377],[119,1.4826],[120,1.2379],[121,0.9884],[122,0.7585],[123,0.5421],[124,0.3462],[125,0.1686],[126,0.0049],[127,-0.1405],[128,-0.2719],[129,-0.3897],[130,-0.4947],[131,-0.5807],[132,-0.6515],[133,-0.7058],[134,-0.7402],[135,-0.7521],[136,-0.7565],[139,-0.7599],[144,-0.773],[147,-0.7841],[149,-0.787],[152,-0.7861],[155,-0.7773],[156,-0.7723],[157,-0.7622],[159,-0.7358],[160,-0.7198],[162,-0.6823],[165,-0.616],[167,-0.5693],[169,-0.5253],[171,-0.4844],[172,-0.4658],[173,-0.4485],[175,-0.4205],[177,-0.4017],[178,-0.3938],[180,-0.3848],[181,-0.383],[183,-0.3817],[184,-0.3817],[185,-0.3853],[186,-0.386],[187,-0.3882],[190,-0.4009],[192,-0.4122],[195,-0.4322],[197,-0.4467],[200,-0.4711],[201,-0.4805],[202,-0.4919],[203,-0.5058],[205,-0.538],[206,-0.5554],[207,-0.575],[209,-0.6196],[210,-0.6441],[211,-0.6772],[212,-0.7242],[213,-0.7833],[214,-0.853],[215,-0.9322],[216,-1.0206],[217,-1.1189],[218,-1.2247],[219,-1.3371],[221,-1.5792],[223,-1.8307],[224,-1.9663],[225,-2.0651],[226,-2.0984],[227,-2.0763],[228,-2.008],[229,-1.9011],[230,-1.768],[231,-1.6124],[232,-1.4413],[233,-1.2555],[234,-1.0628],[237,-0.4609],[238,-0.2647],[239,-0.0759],[240,0.1029],[241,0.2659],[242,0.4132],[243,0.5357],[244,0.6193],[245,0.6313],[246,0.5672],[247,0.4336],[248,0.2409],[249,-0.0096],[250,-0.3026],[251,-0.6314],[252,-0.9915],[253,-1.3686],[255,-2.1443],[256,-2.5248],[257,-2.8827],[258,-3.2138],[259,-3.5027],[260,-3.7429],[261,-3.9207],[262,-4.0311],[263,-4.0583],[264,-4.0287],[265,-3.965],[266,-3.8742],[267,-3.7622],[268,-3.6272],[269,-3.4746],[270,-3.3077],[271,-3.1335],[274,-2.5788],[275,-2.4017],[276,-2.232],[277,-2.0728],[278,-1.9345],[279,-1.8209],[284,-1.3151],[285,-1.219],[286,-1.1264],[287,-1.0371],[288,-0.9502],[289,-0.8688],[290,-0.7911],[291,-0.7211],[292,-0.6559],[293,-0.596],[294,-0.5458],[295,-0.5007],[296,-0.4654],[297,-0.4396],[298,-0.4153],[300,-0.3738],[301,-0.3508],[302,-0.3304],[304,-0.2873],[306,-0.2477],[307,-0.2293],[309,-0.1934],[323,-0.3844],[324,-0.503],[325,-0.7626],[326,-0.8366],[327,-0.9326],[329,-1.0068],[330,-1.0613],[331,-1.1239],[332,-0.9065],[333,-1.2458],[337,-1.4839],[338,-1.5473],[340,-1.6862],[342,-1.8418],[343,-1.9238],[344,-2.0154],[346,-2.1742],[349,-2.3902],[351,-2.5104],[352,-2.5617],[353,-2.6059],[354,-2.6434],[356,-2.6917],[357,-2.7057],[358,-2.7085],[359,-2.705],[360,-2.6959],[362,-2.6594],[365,-2.5825],[368,-2.4871],[371,-2.3845],[374,-2.2902],[377,-2.2083],[380,-2.144],[381,-2.1272],[382,-2.1029],[389,-1.9761],[390,-1.9556],[402,-1.7493],[405,-1.6856],[406,-1.6599],[407,-1.6306],[408,-1.5945],[409,-1.5506],[410,-1.4997],[411,-1.4411],[412,-1.3535],[413,-1.2292],[414,-1.0594],[415,-0.8303],[416,-0.5351],[417,-0.1716],[418,0.2609],[419,0.7669],[420,1.3348],[421,1.9589],[422,2.5573],[423,3.0331],[424,3.3543],[425,3.5699],[426,3.7627],[427,3.9259],[428,4.0515],[429,4.1413],[430,4.1887],[431,4.1903],[432,4.1501],[433,4.0955],[434,3.9337],[435,3.7434],[436,3.5272],[437,3.301],[438,3.0561],[440,2.5387],[443,1.7377],[444,1.4826],[445,1.2379],[446,0.9884],[447,0.7585],[448,0.5421],[449,0.3462],[450,0.1686],[451,0.0049],[452,-0.1405],[453,-0.2719],[454,-0.3897],[455,-0.4947],[456,-0.5807],[457,-0.6515],[458,-0.7058],[459,-0.7402],[460,-0.7521],[461,-0.7565],[464,-0.7599],[469,-0.773],[472,-0.7841],[474,-0.787],[477,-0.7861],[480,-0.7773],[481,-0.7723],[482,-0.7622],[484,-0.7358],[485,-0.7198],[487,-0.6823],[490,-0.616],[492,-0.5693],[494,-0.5253],[496,-0.4844],[497,-0.4658],[498,-0.4485],[500,-0.4205],[502,-0.4017],[503,-0.3938],[505,-0.3848],[506,-0.383],[508,-0.3817],[509,-0.3817],[510,-0.3853],[511,-0.386],[512,-0.3882],[515,-0.4009],[517,-0.4122],[520,-0.4322],[522,-0.4467],[525,-0.4711],[526,-0.4805],[527,-0.4919],[528,-0.5058],[530,-0.538],[531,-0.5554],[532,-0.575],[534,-0.6196],[535,-0.6441],[536,-0.6772],[537,-0.7242],[538,-0.7833],[539,-0.853],[540,-0.9322],[541,-1.0206],[542,-1.1189],[543,-1.2247],[544,-1.3371],[546,-1.5792],[548,-1.8307],[549,-1.9663],[550,-2.0651],[551,-2.0984],[552,-2.0763],[553,-2.008],[554,-1.9011],[555,-1.768],[556,-1.6124],[557,-1.4413],[558,-1.2555],[559,-1.0628],[562,-0.4609],[563,-0.2647],[564,-0.0759],[565,0.1029],[566,0.2659],[567,0.4132],[568,0.5357],[569,0.6193],[570,0.6313],[571,0.5672],[572,0.4336],[573,0.2409],[574,-0.0096],[575,-0.3026],[576,-0.6314],[577,-0.9915],[578,-1.3686],[580,-2.1443],[581,-2.5248],[582,-2.8827],[583,-3.2138],[584,-3.5027],[585,-3.7429],[586,-3.9207],[587,-4.0311],[588,-4.0583],[589,-4.0287],[590,-3.965],[591,-3.8742],[592,-3.7622],[593,-3.6272],[594,-3.4746],[595,-3.3077],[596,-3.1335],[599,-2.5788],[600,-2.4017],[601,-2.232],[602,-2.0728],[603,-1.9345],[604,-1.8209],[609,-1.3151],[610,-1.219],[611,-1.1264],[612,-1.0371],[613,-0.9502],[614,-0.8688],[615,-0.7911],[616,-0.7211],[617,-0.6559],[618,-0.596],[619,-0.5458],[620,-0.5007],[621,-0.4654],[622,-0.4396],[623,-0.4153],[625,-0.3738],[626,-0.3508],[627,-0.3304],[629,-0.2873],[631,-0.2477],[632,-0.2293],[634,-0.1934],[676,-0.3844],[677,-0.503],[678,-0.7626],[679,-0.8366],[680,-0.9326],[682,-1.0068],[683,-1.0613],[684,-1.1239],[685,-0.9065],[686,-1.2458],[690,-1.4839],[691,-1.5473],[693,-1.6862],[695,-1.8418],[696,-1.9238],[697,-2.0154],[699,-2.1742],[702,-2.3902],[704,-2.5104],[705,-2.5617],[706,-2.6059],[707,-2.6434],[709,-2.6917],[710,-2.7057],[711,-2.7085],[712,-2.705],[713,-2.6959],[715,-2.6594],[718,-2.5825],[721,-2.4871],[724,-2.3845],[727,-2.2902],[730,-2.2083],[733,-2.144],[734,-2.1272],[735,-2.1029],[742,-1.9761],[743,-1.9556],[755,-1.7493],[757,-1.7078],[759,-1.6599],[770,-1.219],[771,-1.1264],[772,-1.0371],[773,-0.9502],[774,-0.8688],[775,-0.7911],[776,-0.7211],[777,-0.6559],[778,-0.596],[779,-0.5458],[780,-0.5007],[781,-0.4654],[782,-0.4396],[783,-0.4153],[785,-0.3738],[786,-0.3508],[787,-0.3304],[789,-0.2873],[791,-0.2477],[792,-0.2293],[794,-0.1934],[795,-0.177],[797,-0.1491],[798,-0.1362],[799,-0.1246],[800,-0.1124],[801,-0.1039],[802,-0.094],[803,-0.0858],[837,-0.3844],[887,-0.177],[889,-0.1491],[890,-0.1362],[891,-0.1246],[892,-0.1124],[893,-0.1039],[894,-0.094],[895,-0.0858],[929,-0.3844]]
       }
       leg_l_pv_ik_ctrl.Location.Y {
        type=MovieSceneScriptingFloatChannel
        default=0
        thinned_from=751
        keys [[-2,4.0213],[-1,3.5433],[0,3.926],[1,4.3448],[2,4.5851],[4,4.6924],[6,4.8193],[7,1.763],[8,4.9317],[18,5.3733],[19,5.4783],[30,5.9353],[44,6.3113],[54,6.4644],[65,6.4654],[77,6.298],[80,6.2283],[82,6.1593],[85,6.0115],[87,5.8777],[89,5.7038],[91,5.4877],[92,5.3624],[93,5.2222],[94,5.0686],[95,4.8999],[98,4.3335],[99,4.1744],[100,4.0475],[101,3.9533],[102,3.9027],[103,3.9088],[104,3.9751],[105,4.1061],[106,4.3065],[107,4.586],[108,5.1511],[109,5.2831],[110,5.4375],[117,6.6681],[119,6.9873],[120,7.1292],[121,6.9806],[122,6.9818],[125,7.0549],[139,7.5017],[140,7.5207],[149,7.7261],[156,7.7725],[163,7.5481],[174,7.129],[197,6.6626],[201,6.6172],[202,6.6212],[204,6.6651],[210,6.9332],[212,7.06],[215,7.315],[217,7.5272],[224,8.3663],[226,8.5701],[228,8.7299],[230,8.8336],[231,8.8611],[233,8.8638],[235,8.8001],[237,8.6719],[239,8.4894],[241,8.2592],[244,7.8311],[245,7.7073],[246,7.6035],[247,7.5198],[248,7.4536],[250,7.366],[262,7.1166],[264,7.0985],[267,7.1238],[277,7.3989],[284,7.5083],[290,7.5088],[299,7.3574],[305,7.1514],[309,6.9733],[323,4.0213],[324,3.5433],[325,3.926],[326,4.3448],[327,4.5851],[329,4.6924],[331,4.8193],[332,1.763],[333,4.9317],[343,5.3733],[344,5.4783],[355,5.9353],[369,6.3113],[379,6.4644],[390,6.4654],[402,6.298],[405,6.2283],[407,6.1593],[410,6.0115],[412,5.8777],[414,5.7038],[416,5.4877],[417,5.3624],[418,5.2222],[419,5.0686],[420,4.8999],[423,4.3335],[424,4.1744],[425,4.0475],[426,3.9533],[427,3.9027],[428,3.9088],[429,3.9751],[430,4.1061],[431,4.3065],[432,4.586],[433,5.1511],[434,5.2831],[435,5.4375],[442,6.6681],[444,6.9873],[445,7.1292],[446,6.9806],[447,6.9818],[450,7.0549],[464,7.5017],[465,7.5207],[474,7.7261],[481,7.7725],[488,7.5481],[499,7.129],[522,6.6626],[526,6.6172],[527,6.6212],[529,6.6651],[535,6.9332],[537,7.06],[540,7.315],[542,7.5272],[549,8.3663],[551,8.5701],[553,8.7299],[555,8.8336],[556,8.8611],[558,8.8638],[560,8.8001],[562,8.6719],[564,8.4894],[566,8.2592],[569,7.8311],[570,7.7073],[571,7.6035],[572,7.5198],[573,7.4536],[575,7.366],[587,7.1166],[589,7.0985],[592,7.1238],[602,7.3989],[609,7.5083],[615,7.5088],[624,7.3574],[630,7.1514],[634,6.9733],[676,4.0213],[677,3.5433],[678,3.926],[679,4.3448],[680,4.5851],[682,4.6924],[684,4.8193],[685,1.763],[686,4.9317],[696,5.3733],[697,5.4783],[708,5.9353],[722,6.3113],[732,6.4644],[743,6.4654],[755,6.298],[759,6.1967],[770,7.5138],[777,7.4885],[785,7.3308],[803,6.5559],[837,4.0213],[887,6.9272],[895,6.5559],[929,4.0213]]
       }
       leg_l_pv_ik_ctrl.Location.Z {
        type=MovieSceneScriptingFloatChannel
        default=0
        thinned_from=751
        keys [[-2,4.2355],[-1,3.6665],[0,4.0815],[1,4.569],[2,4.8483],[4,4.971],[6,5.1144],[7,1.6739],[8,5.2406],[12,5.4542],[18,5.7323],[19,5.8582],[31,6.4448],[47,6.9975],[54,7.1357],[62,7.1693],[71,7.0756],[79,6.8906],[81,6.8198],[84,6.6574],[86,6.5151],[89,6.2336],[92,5.9002],[97,5.2978],[99,5.0469],[100,4.9422],[101,4.8733],[102,4.8509],[103,4.8879],[104,4.9888],[105,5.1582],[106,5.4021],[107,5.7366],[108,6.4359],[109,6.5654],[110,6.7189],[118,8.1609],[120,8.4856],[121,8.232],[122,8.1951],[124,8.1892],[131,8.3869],[139,8.7168],[140,8.7438],[147,8.9921],[153,9.0993],[156,9.1086],[160,8.9529],[173,8.2698],[192,7.6861],[200,7.5216],[202,7.5156],[204,7.5719],[206,7.6696],[210,7.9265],[211,8.0021],[213,8.1989],[215,8.4291],[217,8.7119],[224,9.8684],[227,10.3045],[229,10.5268],[230,10.6082],[231,10.6669],[232,10.7001],[233,10.7094],[234,10.6943],[235,10.6561],[236,10.5937],[238,10.4052],[240,10.1404],[241,9.9806],[244,9.3956],[246,9.0527],[248,8.7841],[250,8.5805],[253,8.3634],[260,8.005],[262,7.9372],[264,7.912],[266,7.9297],[269,8.0295],[275,8.3518],[283,8.6396],[288,8.7185],[293,8.6992],[299,8.5559],[305,8.2803],[309,8.0404],[323,4.2355],[324,3.6665],[325,4.0815],[326,4.569],[327,4.8483],[329,4.971],[331,5.1144],[332,1.6739],[333,5.2406],[337,5.4542],[343,5.7323],[344,5.8582],[356,6.4448],[372,6.9975],[379,7.1357],[387,7.1693],[396,7.0756],[404,6.8906],[406,6.8198],[409,6.6574],[411,6.5151],[414,6.2336],[417,5.9002],[422,5.2978],[424,5.0469],[425,4.9422],[426,4.8733],[427,4.8509],[428,4.8879],[429,4.9888],[430,5.1582],[431,5.4021],[432,5.7366],[433,6.4359],[434,6.5654],[435,6.7189],[443,8.1609],[445,8.4856],[446,8.232],[447,8.1951],[449,8.1892],[456,8.3869],[464,8.7168],[465,8.7438],[472,8.9921],[478,9.0993],[481,9.1086],[485,8.9529],[498,8.2698],[517,7.6861],[525,7.5216],[527,7.5156],[529,7.5719],[531,7.6696],[535,7.9265],[536,8.0021],[538,8.1989],[540,8.4291],[542,8.7119],[549,9.8684],[552,10.3045],[554,10.5268],[555,10.6082],[556,10.6669],[557,10.7001],[558,10.7094],[559,10.6943],[560,10.6561],[561,10.5937],[563,10.4052],[565,10.1404],[566,9.9806],[569,9.3956],[571,9.0527],[573,8.7841],[575,8.5805],[578,8.3634],[585,8.005],[587,7.9372],[589,7.912],[591,7.9297],[594,8.0295],[600,8.3518],[608,8.6396],[613,8.7185],[618,8.6992],[624,8.5559],[630,8.2803],[634,8.0404],[676,4.2355],[677,3.6665],[678,4.0815],[679,4.569],[680,4.8483],[682,4.971],[684,5.1144],[685,1.6739],[686,5.2406],[690,5.4542],[696,5.7323],[697,5.8582],[709,6.4448],[725,6.9975],[732,7.1357],[740,7.1693],[749,7.0756],[755,6.9466],[757,6.8906],[759,6.8198],[770,8.6816],[775,8.7232],[780,8.6659],[785,8.5208],[792,8.1643],[803,7.4797],[837,4.2355],[887,7.9784],[895,7.4797],[929,4.2355]]
       }
       leg_l_pv_ik_ctrl.Rotation.X {
        type=MovieSceneScriptingFloatChannel
        default=0
        thinned_from=751
        keys [[-2,-5.7093],[-1,-4.9733],[0,-5.5129],[1,-6.14],[2,-6.4969],[4,-6.6533],[6,-6.8358],[7,-2.4519],[8,-6.9962],[12,-7.2669],[18,-7.6186],[19,-7.7766],[31,-8.5097],[47,-9.1883],[54,-9.3564],[63,-9.3915],[74,-9.2109],[80,-9.0153],[82,-8.9086],[85,-8.6827],[87,-8.4829],[89,-8.2387],[92,-7.8186],[97,-7.0476],[99,-6.7259],[100,-6.592],[101,-6.5033],[102,-6.474],[103,-6.5206],[104,-6.6484],[105,-6.863],[106,-7.1712],[107,-7.5918],[108,-8.4631],[109,-8.6241],[110,-8.8144],[118,-10.5802],[120,-10.9726],[121,-10.6694],[122,-10.6261],[124,-10.6211],[131,-10.8634],[139,-11.2591],[140,-11.2914],[148,-11.6165],[155,-11.7266],[156,-11.7257],[162,-11.4182],[174,-10.6704],[197,-9.886],[201,-9.8056],[202,-9.8106],[204,-9.8791],[206,-9.998],[210,-10.3095],[212,-10.5117],[215,-10.9158],[217,-11.2552],[224,-12.629],[227,-13.1401],[229,-13.3984],[231,-13.5599],[232,-13.5977],[233,-13.6074],[235,-13.5436],[237,-13.3699],[239,-13.1023],[241,-12.7495],[244,-12.0588],[246,-11.6525],[248,-11.3344],[250,-11.094],[253,-10.8393],[260,-10.4213],[262,-10.3412],[264,-10.3108],[266,-10.3312],[269,-10.4493],[275,-10.831],[283,-11.1704],[288,-11.2622],[293,-11.2373],[299,-11.065],[305,-10.7338],[309,-10.4445],[323,-5.7093],[324,-4.9733],[325,-5.5129],[326,-6.14],[327,-6.4969],[329,-6.6533],[331,-6.8358],[332,-2.4519],[333,-6.9962],[337,-7.2669],[343,-7.6186],[344,-7.7766],[356,-8.5097],[372,-9.1883],[379,-9.3564],[388,-9.3915],[399,-9.2109],[405,-9.0153],[407,-8.9086],[410,-8.6827],[412,-8.4829],[414,-8.2387],[417,-7.8186],[422,-7.0476],[424,-6.7259],[425,-6.592],[426,-6.5033],[427,-6.474],[428,-6.5206],[429,-6.6484],[430,-6.863],[431,-7.1712],[432,-7.5918],[433,-8.4631],[434,-8.6241],[435,-8.8144],[443,-10.5802],[445,-10.9726],[446,-10.6694],[447,-10.6261],[449,-10.6211],[456,-10.8634],[464,-11.2591],[465,-11.2914],[473,-11.6165],[480,-11.7266],[481,-11.7257],[487,-11.4182],[499,-10.6704],[522,-9.886],[526,-9.8056],[527,-9.8106],[529,-9.8791],[531,-9.998],[535,-10.3095],[537,-10.5117],[540,-10.9158],[542,-11.2552],[549,-12.629],[552,-13.1401],[554,-13.3984],[556,-13.5599],[557,-13.5977],[558,-13.6074],[560,-13.5436],[562,-13.3699],[564,-13.1023],[566,-12.7495],[569,-12.0588],[571,-11.6525],[573,-11.3344],[575,-11.094],[578,-10.8393],[585,-10.4213],[587,-10.3412],[589,-10.3108],[591,-10.3312],[594,-10.4493],[600,-10.831],[608,-11.1704],[613,-11.2622],[618,-11.2373],[624,-11.065],[630,-10.7338],[634,-10.4445],[676,-5.7093],[677,-4.9733],[678,-5.5129],[679,-6.14],[680,-6.4969],[682,-6.6533],[684,-6.8358],[685,-2.4519],[686,-6.9962],[690,-7.2669],[696,-7.6186],[697,-7.7766],[709,-8.5097],[725,-9.1883],[732,-9.3564],[741,-9.3915],[752,-9.2109],[757,-9.0535],[759,-8.9663],[770,-11.2195],[775,-11.267],[781,-11.1705],[786,-10.9728],[803,-9.7648],[837,-5.7093],[887,-10.3696],[895,-9.7648],[929,-5.7093]]
       }
       leg_l_pv_ik_ctrl.Rotation.Y {
        type=MovieSceneScriptingFloatChannel
        default=0
        thinned_from=751
        keys [[-2,0.0378],[-1,0.0206],[0,0.014],[1,0.0197],[2,0.0203],[4,0.0189],[5,0.0174],[6,0.0156],[7,-1.2474],[8,0.0117],[9,0.0098],[12,0.0034],[13,0.0009],[15,-0.0046],[17,-0.0111],[18,-0.0146],[19,-0.0167],[21,-0.0232],[23,-0.0289],[24,-0.0316],[26,-0.0358],[27,-0.0374],[28,-0.0387],[29,-0.0396],[31,-0.0399],[32,-0.0397],[33,-0.0388],[34,-0.0376],[35,-0.0361],[36,-0.0342],[38,-0.0298],[40,-0.0249],[42,-0.0194],[43,-0.0168],[44,-0.014],[48,-0.0036],[49,-0.0012],[52,0.0049],[55,0.0095],[56,0.0107],[57,0.0122],[61,0.0165],[65,0.02],[70,0.023],[82,0.0277],[83,0.0282],[84,0.0291],[85,0.0301],[86,0.0314],[87,0.034],[88,0.0383],[89,0.045],[90,0.0547],[91,0.0678],[92,0.0842],[93,0.1039],[94,0.1271],[95,0.1529],[96,0.1808],[97,0.2067],[98,0.2263],[99,0.2387],[101,0.254],[102,0.2608],[103,0.2669],[104,0.2725],[105,0.2774],[107,0.286],[108,0.2972],[109,0.293],[110,0.288],[112,0.2765],[115,0.2554],[119,0.2249],[120,0.2176],[121,0.1991],[122,0.1871],[123,0.1764],[124,0.1669],[125,0.1587],[126,0.1513],[127,0.1447],[128,0.1389],[129,0.1339],[130,0.1295],[131,0.1262],[132,0.1237],[133,0.122],[134,0.1215],[135,0.1221],[139,0.1264],[140,0.127],[141,0.1284],[145,0.1318],[151,0.1354],[156,0.1371],[178,0.1271],[184,0.1227],[200,0.1084],[201,0.1076],[202,0.1072],[203,0.107],[204,0.1071],[205,0.1074],[210,0.1103],[211,0.1107],[212,0.1108],[213,0.1106],[214,0.1102],[218,0.1071],[219,0.1062],[221,0.1039],[223,0.101],[224,0.099],[225,0.0987],[226,0.1016],[227,0.1072],[228,0.1146],[229,0.1235],[230,0.1329],[233,0.1622],[234,0.1713],[235,0.18],[236,0.1879],[237,0.195],[238,0.2013],[239,0.2066],[240,0.2109],[241,0.2138],[242,0.2154],[243,0.2154],[244,0.2135],[245,0.2089],[246,0.2013],[247,0.1908],[248,0.178],[249,0.1627],[250,0.1455],[251,0.1267],[252,0.1063],[253,0.0849],[256,0.0184],[257,-0.0029],[258,-0.0229],[259,-0.0407],[260,-0.0558],[261,-0.0672],[262,-0.0745],[263,-0.0766],[264,-0.0751],[265,-0.0713],[266,-0.0656],[267,-0.0584],[268,-0.0497],[269,-0.0397],[270,-0.0288],[271,-0.0173],[273,0.0068],[274,0.0184],[275,0.0295],[276,0.0399],[277,0.0498],[278,0.0583],[279,0.0655],[281,0.0783],[283,0.0905],[284,0.0963],[286,0.1071],[288,0.1167],[289,0.121],[290,0.125],[291,0.1284],[292,0.1314],[293,0.1341],[294,0.1362],[295,0.138],[296,0.1392],[298,0.1401],[301,0.1403],[304,0.1395],[307,0.1377],[309,0.1362],[323,0.0378],[324,0.0206],[325,0.014],[326,0.0197],[327,0.0203],[329,0.0189],[330,0.0174],[331,0.0156],[332,-1.2474],[333,0.0117],[334,0.0098],[337,0.0034],[338,0.0009],[340,-0.0046],[342,-0.0111],[343,-0.0146],[344,-0.0167],[346,-0.0232],[348,-0.0289],[349,-0.0316],[351,-0.0358],[352,-0.0374],[353,-0.0387],[354,-0.0396],[356,-0.0399],[357,-0.0397],[358,-0.0388],[359,-0.0376],[360,-0.0361],[361,-0.0342],[363,-0.0298],[365,-0.0249],[367,-0.0194],[368,-0.0168],[369,-0.014],[373,-0.0036],[374,-0.0012],[377,0.0049],[380,0.0095],[381,0.0107],[382,0.0122],[386,0.0165],[390,0.02],[395,0.023],[407,0.0277],[408,0.0282],[409,0.0291],[410,0.0301],[411,0.0314],[412,0.034],[413,0.0383],[414,0.045],[415,0.0547],[416,0.0678],[417,0.0842],[418,0.1039],[419,0.1271],[420,0.1529],[421,0.1808],[422,0.2067],[423,0.2263],[424,0.2387],[426,0.254],[427,0.2608],[428,0.2669],[429,0.2725],[430,0.2774],[432,0.286],[433,0.2972],[434,0.293],[435,0.288],[437,0.2765],[440,0.2554],[444,0.2249],[445,0.2176],[446,0.1991],[447,0.1871],[448,0.1764],[449,0.1669],[450,0.1587],[451,0.1513],[452,0.1447],[453,0.1389],[454,0.1339],[455,0.1295],[456,0.1262],[457,0.1237],[458,0.122],[459,0.1215],[460,0.1221],[464,0.1264],[465,0.127],[466,0.1284],[470,0.1318],[476,0.1354],[481,0.1371],[503,0.1271],[509,0.1227],[525,0.1084],[526,0.1076],[527,0.1072],[528,0.107],[529,0.1071],[530,0.1074],[535,0.1103],[536,0.1107],[537,0.1108],[538,0.1106],[539,0.1102],[543,0.1071],[544,0.1062],[546,0.1039],[548,0.101],[549,0.099],[550,0.0987],[551,0.1016],[552,0.1072],[553,0.1146],[554,0.1235],[555,0.1329],[558,0.1622],[559,0.1713],[560,0.18],[561,0.1879],[562,0.195],[563,0.2013],[564,0.2066],[565,0.2109],[566,0.2138],[567,0.2154],[568,0.2154],[569,0.2135],[570,0.2089],[571,0.2013],[572,0.1908],[573,0.178],[574,0.1627],[575,0.1455],[576,0.1267],[577,0.1063],[578,0.0849],[581,0.0184],[582,-0.0029],[583,-0.0229],[584,-0.0407],[585,-0.0558],[586,-0.0672],[587,-0.0745],[588,-0.0766],[589,-0.0751],[590,-0.0713],[591,-0.0656],[592,-0.0584],[593,-0.0497],[594,-0.0397],[595,-0.0288],[596,-0.0173],[598,0.0068],[599,0.0184],[600,0.0295],[601,0.0399],[602,0.0498],[603,0.0583],[604,0.0655],[606,0.0783],[608,0.0905],[609,0.0963],[611,0.1071],[613,0.1167],[614,0.121],[615,0.125],[616,0.1284],[617,0.1314],[618,0.1341],[619,0.1362],[620,0.138],[621,0.1392],[623,0.1401],[626,0.1403],[629,0.1395],[632,0.1377],[634,0.1362],[676,0.0378],[677,0.0206],[678,0.014],[679,0.0197],[680,0.0203],[682,0.0189],[683,0.0174],[684,0.0156],[685,-1.2474],[686,0.0117],[687,0.0098],[690,0.0034],[691,0.0009],[693,-0.0046],[695,-0.0111],[696,-0.0146],[697,-0.0167],[699,-0.0232],[701,-0.0289],[702,-0.0316],[704,-0.0358],[705,-0.0374],[706,-0.0387],[707,-0.0396],[709,-0.0399],[710,-0.0397],[711,-0.0388],[712,-0.0376],[713,-0.0361],[714,-0.0342],[716,-0.0298],[718,-0.0249],[720,-0.0194],[721,-0.0168],[722,-0.014],[726,-0.0036],[727,-0.0012],[730,0.0049],[733,0.0095],[734,0.0107],[735,0.0122],[739,0.0165],[743,0.02],[748,0.023],[759,0.0273],[770,0.1018],[771,0.1071],[773,0.1167],[774,0.121],[775,0.125],[776,0.1284],[777,0.1314],[778,0.1341],[779,0.1362],[780,0.138],[781,0.1392],[783,0.1401],[786,0.1403],[789,0.1395],[792,0.1377],[795,0.1354],[800,0.1303],[803,0.1269],[837,0.0378],[887,0.1354],[892,0.1303],[895,0.1269],[929,0.0378]]
       }
       leg_l_pv_ik_ctrl.Rotation.Z {
        type=MovieSceneScriptingFloatChannel
        default=0
        thinned_from=751
        keys [[-2,0.5277],[-1,0.6735],[0,0.9974],[1,1.0876],[2,1.2044],[4,1.2947],[5,1.3609],[6,1.4368],[7,0.0398],[8,1.5843],[12,1.871],[13,1.9471],[15,2.1136],[17,2.2998],[18,2.3977],[19,2.5052],[21,2.6938],[24,2.9486],[26,3.0893],[27,3.149],[28,3.2002],[29,3.2434],[31,3.2979],[32,3.3133],[33,3.3154],[35,3.2979],[37,3.2521],[40,3.1577],[49,2.8044],[52,2.7063],[55,2.6295],[56,2.6095],[57,2.5806],[77,2.1666],[80,2.0923],[81,2.0625],[82,2.0283],[83,1.9861],[84,1.9347],[85,1.8749],[86,1.8059],[87,1.7022],[88,1.5543],[89,1.3516],[90,1.0767],[91,0.721],[92,0.281],[93,-0.2448],[94,-0.8631],[95,-1.5607],[96,-2.3322],[97,-3.0776],[98,-3.6762],[99,-4.0859],[100,-4.3643],[101,-4.6119],[102,-4.819],[103,-4.973],[104,-5.0754],[105,-5.1177],[106,-5.0951],[107,-5.0123],[108,-4.8794],[109,-4.67],[110,-4.4252],[111,-4.1489],[112,-3.8619],[113,-3.5542],[115,-2.9118],[116,-2.5813],[118,-1.9352],[119,-1.6281],[120,-1.3355],[121,-1.0511],[122,-0.7823],[123,-0.5285],[124,-0.2987],[125,-0.0904],[126,0.1016],[127,0.272],[128,0.4258],[129,0.5637],[130,0.6865],[131,0.7869],[132,0.8696],[133,0.9329],[134,0.973],[135,0.987],[136,0.9923],[139,0.9966],[144,1.0123],[147,1.0253],[150,1.0291],[153,1.0257],[155,1.0178],[156,1.012],[157,1.0003],[159,0.9696],[160,0.9508],[162,0.907],[165,0.8295],[167,0.7746],[169,0.7228],[171,0.6745],[172,0.6524],[173,0.6319],[175,0.5986],[177,0.5761],[178,0.5665],[180,0.5554],[182,0.5517],[184,0.5508],[185,0.5547],[186,0.5554],[187,0.5577],[190,0.572],[192,0.585],[195,0.6081],[197,0.6248],[200,0.6532],[201,0.6642],[202,0.6777],[203,0.6942],[205,0.7327],[206,0.7534],[207,0.7768],[209,0.8298],[210,0.8588],[211,0.8979],[212,0.9532],[213,1.0225],[214,1.1039],[215,1.1963],[216,1.299],[217,1.4126],[218,1.5345],[219,1.6635],[221,1.9394],[223,2.2239],[224,2.3766],[225,2.4869],[226,2.5226],[227,2.4956],[228,2.4166],[229,2.2944],[230,2.1432],[231,1.9669],[232,1.7734],[233,1.5634],[234,1.3456],[237,0.663],[238,0.4392],[239,0.2229],[240,0.017],[241,-0.1721],[242,-0.3444],[243,-0.4894],[244,-0.5903],[245,-0.6085],[246,-0.5379],[247,-0.3861],[248,-0.1646],[249,0.1249],[250,0.4646],[251,0.8465],[252,1.2652],[253,1.7041],[255,2.6073],[256,3.0506],[257,3.468],[258,3.8544],[259,4.192],[260,4.473],[261,4.6815],[262,4.8112],[263,4.8439],[264,4.8097],[265,4.7354],[266,4.6292],[267,4.4977],[268,4.3393],[269,4.1601],[270,3.9642],[271,3.7598],[274,3.1107],[275,2.9041],[276,2.7063],[277,2.5209],[278,2.36],[279,2.2277],[284,1.6402],[285,1.5286],[286,1.4213],[287,1.3178],[288,1.2171],[289,1.1228],[290,1.0327],[291,0.9515],[292,0.8758],[293,0.8063],[294,0.7479],[295,0.6955],[296,0.6543],[297,0.6241],[298,0.5956],[300,0.5468],[301,0.5197],[302,0.4956],[304,0.4443],[306,0.3969],[309,0.3314],[323,0.5277],[324,0.6735],[325,0.9974],[326,1.0876],[327,1.2044],[329,1.2947],[330,1.3609],[331,1.4368],[332,0.0398],[333,1.5843],[337,1.871],[338,1.9471],[340,2.1136],[342,2.2998],[343,2.3977],[344,2.5052],[346,2.6938],[349,2.9486],[351,3.0893],[352,3.149],[353,3.2002],[354,3.2434],[356,3.2979],[357,3.3133],[358,3.3154],[360,3.2979],[362,3.2521],[365,3.1577],[374,2.8044],[377,2.7063],[380,2.6295],[381,2.6095],[382,2.5806],[402,2.1666],[405,2.0923],[406,2.0625],[407,2.0283],[408,1.9861],[409,1.9347],[410,1.8749],[411,1.8059],[412,1.7022],[413,1.5543],[414,1.3516],[415,1.0767],[416,0.721],[417,0.281],[418,-0.2448],[419,-0.8631],[420,-1.5607],[421,-2.3322],[422,-3.0776],[423,-3.6762],[424,-4.0859],[425,-4.3643],[426,-4.6119],[427,-4.819],[428,-4.973],[429,-5.0754],[430,-5.1177],[431,-5.0951],[432,-5.0123],[433,-4.8794],[434,-4.67],[435,-4.4252],[436,-4.1489],[437,-3.8619],[438,-3.5542],[440,-2.9118],[441,-2.5813],[443,-1.9352],[444,-1.6281],[445,-1.3355],[446,-1.0511],[447,-0.7823],[448,-0.5285],[449,-0.2987],[450,-0.0904],[451,0.1016],[452,0.272],[453,0.4258],[454,0.5637],[455,0.6865],[456,0.7869],[457,0.8696],[458,0.9329],[459,0.973],[460,0.987],[461,0.9923],[464,0.9966],[469,1.0123],[472,1.0253],[475,1.0291],[478,1.0257],[480,1.0178],[481,1.012],[482,1.0003],[484,0.9696],[485,0.9508],[487,0.907],[490,0.8295],[492,0.7746],[494,0.7228],[496,0.6745],[497,0.6524],[498,0.6319],[500,0.5986],[502,0.5761],[503,0.5665],[505,0.5554],[507,0.5517],[509,0.5508],[510,0.5547],[511,0.5554],[512,0.5577],[515,0.572],[517,0.585],[520,0.6081],[522,0.6248],[525,0.6532],[526,0.6642],[527,0.6777],[528,0.6942],[530,0.7327],[531,0.7534],[532,0.7768],[534,0.8298],[535,0.8588],[536,0.8979],[537,0.9532],[538,1.0225],[539,1.1039],[540,1.1963],[541,1.299],[542,1.4126],[543,1.5345],[544,1.6635],[546,1.9394],[548,2.2239],[549,2.3766],[550,2.4869],[551,2.5226],[552,2.4956],[553,2.4166],[554,2.2944],[555,2.1432],[556,1.9669],[557,1.7734],[558,1.5634],[559,1.3456],[562,0.663],[563,0.4392],[564,0.2229],[565,0.017],[566,-0.1721],[567,-0.3444],[568,-0.4894],[569,-0.5903],[570,-0.6085],[571,-0.5379],[572,-0.3861],[573,-0.1646],[574,0.1249],[575,0.4646],[576,0.8465],[577,1.2652],[578,1.7041],[580,2.6073],[581,3.0506],[582,3.468],[583,3.8544],[584,4.192],[585,4.473],[586,4.6815],[587,4.8112],[588,4.8439],[589,4.8097],[590,4.7354],[591,4.6292],[592,4.4977],[593,4.3393],[594,4.1601],[595,3.9642],[596,3.7598],[599,3.1107],[600,2.9041],[601,2.7063],[602,2.5209],[603,2.36],[604,2.2277],[609,1.6402],[610,1.5286],[611,1.4213],[612,1.3178],[613,1.2171],[614,1.1228],[615,1.0327],[616,0.9515],[617,0.8758],[618,0.8063],[619,0.7479],[620,0.6955],[621,0.6543],[622,0.6241],[623,0.5956],[625,0.5468],[626,0.5197],[627,0.4956],[629,0.4443],[631,0.3969],[634,0.3314],[676,0.5277],[677,0.6735],[678,0.9974],[679,1.0876],[680,1.2044],[682,1.2947],[683,1.3609],[684,1.4368],[685,0.0398],[686,1.5843],[690,1.871],[691,1.9471],[693,2.1136],[695,2.2998],[696,2.3977],[697,2.5052],[699,2.6938],[702,2.9486],[704,3.0893],[705,3.149],[706,3.2002],[707,3.2434],[709,3.2979],[710,3.3133],[711,3.3154],[713,3.2979],[715,3.2521],[718,3.1577],[727,2.8044],[730,2.7063],[733,2.6295],[734,2.6095],[735,2.5806],[755,2.1666],[757,2.1183],[759,2.0625],[770,1.5286],[771,1.4213],[772,1.3178],[773,1.2171],[774,1.1228],[775,1.0327],[776,0.9515],[777,0.8758],[778,0.8063],[779,0.7479],[780,0.6955],[781,0.6543],[782,0.6241],[783,0.5956],[785,0.5468],[786,0.5197],[787,0.4956],[789,0.4443],[791,0.3969],[794,0.3314],[795,0.3114],[797,0.2771],[798,0.2612],[799,0.2467],[800,0.2316],[801,0.2208],[802,0.2083],[803,0.1979],[837,0.5277],[887,0.3114],[889,0.2771],[890,0.2612],[891,0.2467],[892,0.2316],[893,0.2208],[894,0.2083],[895,0.1979],[929,0.5277]]
       }
       leg_l_pv_ik_ctrl.Scale.X {type=MovieSceneScriptingFloatChannel, default=1, const [1,751]}
       leg_l_pv_ik_ctrl.Scale.Y {type=MovieSceneScriptingFloatChannel, default=1, const [1,751]}
       leg_l_pv_ik_ctrl.Scale.Z {type=MovieSceneScriptingFloatChannel, default=1, const [1,751]}
       thigh_r_fk_ctrl.Location.X {type=MovieSceneScriptingFloatChannel, default=0, thinned_from=751, keys [[-2,0],[929,0]]}
       thigh_r_fk_ctrl.Location.Y {type=MovieSceneScriptingFloatChannel, default=0, thinned_from=751, keys [[-2,0],[929,0]]}
       thigh_r_fk_ctrl.Location.Z {type=MovieSceneScriptingFloatChannel, default=0, thinned_from=751, keys [[-2,0],[929,0]]}
       thigh_r_fk_ctrl.Rotation.X {
        type=MovieSceneScriptingFloatChannel
        default=1.2139
        thinned_from=373
        keys [[-2,3.824],[-1,3.8304],[0,3.8455],[12,4.4347],[14,4.5144],[22,4.734],[34,4.5833],[49,4.3083],[68,3.8011],[79,3.5276],[92,2.6785],[93,2.6379],[94,2.6192],[95,2.6176],[147,3.825],[177,4.0216],[203,3.9819],[216,3.9927],[217,4.0071],[218,4.0668],[234,4.7422],[243,4.2162],[249,3.124],[251,2.1412],[252,1.6677],[253,1.2141],[254,0.7802],[255,0.3737],[256,-0.0047],[257,-0.3427],[258,-0.6397],[259,-0.8935],[260,-1.1059],[261,-1.2818],[262,-1.4195],[263,-1.5197],[264,-1.5755],[265,-1.5833],[266,-1.5377],[267,-1.4365],[268,-1.3068],[269,-1.1817],[270,-1.0591],[271,-0.9403],[273,-0.711],[274,-0.6008],[276,-0.3907],[277,-0.2998],[278,-0.2081],[279,-0.1276],[280,-0.0528],[281,0.0167],[282,0.0696],[283,0.1237],[284,0.1627],[285,0.1976],[286,0.2172],[287,0.2298],[288,0.2374],[290,0.2491],[291,0.254],[292,0.2567],[293,0.2574],[296,0.2628],[297,0.2619],[298,0.2634],[299,0.2634],[300,0.2611],[302,0.2605],[303,0.2587],[304,0.2586],[305,0.2565],[306,0.2553],[308,0.252],[309,0.2496],[323,3.824],[324,3.8304],[325,3.8455],[337,4.4347],[339,4.5144],[347,4.734],[359,4.5833],[374,4.3083],[393,3.8011],[404,3.5276],[417,2.6785],[418,2.6379],[419,2.6192],[420,2.6176],[472,3.825],[502,4.0216],[528,3.9819],[541,3.9927],[542,4.0071],[543,4.0668],[559,4.7422],[568,4.2162],[574,3.124],[576,2.1412],[577,1.6677],[578,1.2141],[579,0.7802],[580,0.3737],[581,-0.0047],[582,-0.3427],[583,-0.6397],[584,-0.8935],[585,-1.1059],[586,-1.2818],[587,-1.4195],[588,-1.5197],[589,-1.5755],[590,-1.5833],[591,-1.5377],[592,-1.4365],[593,-1.3068],[594,-1.1817],[595,-1.0591],[596,-0.9403],[598,-0.711],[599,-0.6008],[601,-0.3907],[602,-0.2998],[603,-0.2081],[604,-0.1276],[605,-0.0528],[606,0.0167],[607,0.0696],[608,0.1237],[609,0.1627],[610,0.1976],[611,0.2172],[612,0.2298],[613,0.2374],[615,0.2491],[616,0.254],[617,0.2567],[618,0.2574],[621,0.2628],[622,0.2619],[623,0.2634],[624,0.2634],[625,0.2611],[627,0.2605],[628,0.2587],[629,0.2586],[630,0.2565],[631,0.2553],[633,0.252],[634,0.2496],[659,1.9803],[677,4.0764],[690,4.6808],[692,4.7604],[700,4.98],[712,4.8293],[727,4.5543],[746,4.0471],[804,3.2421],[835,4.0338],[837,4.07],[887,0.2477],[888,0.2445],[890,0.2353],[891,0.2283],[892,0.2208],[893,0.2116],[894,0.1998],[895,0.1867],[926,3.7519],[927,3.7878],[929,3.824]]
       }
       thigh_r_fk_ctrl.Rotation.Y {
        type=MovieSceneScriptingFloatChannel
        default=-1.213
        thinned_from=83
        keys [[-2,-0.6308],[-1,-0.6461],[0,-0.6934],[11,-2.7768],[14,-3.3544],[33,-4.785],[48,-4.6825],[69,-2.5059],[71,-2.3653],[73,-2.2343],[79,-1.9157],[92,-1.3807],[93,-1.3197],[94,-1.2456],[95,-1.1547],[96,-1.0431],[105,0.4304],[127,1.7817],[141,2.1375],[142,2.1273],[198,2.1038],[205,1.8155],[216,0.6798],[223,-0.4566],[240,-3.2143],[250,-5.1105],[264,-6.8279],[280,-7.0017],[307,-5.6158],[309,-5.5298],[323,-0.6308],[324,-0.6461],[325,-0.6934],[336,-2.7768],[339,-3.3544],[358,-4.785],[373,-4.6825],[394,-2.5059],[396,-2.3653],[398,-2.2343],[404,-1.9157],[417,-1.3807],[418,-1.3197],[419,-1.2456],[420,-1.1547],[421,-1.0431],[430,0.4304],[452,1.7817],[466,2.1375],[467,2.1273],[523,2.1038],[530,1.8155],[541,0.6798],[548,-0.4566],[565,-3.2143],[575,-5.1105],[589,-6.8279],[605,-7.0017],[632,-5.6158],[634,-5.5298],[676,-3.2763],[677,-3.2915],[689,-3.707],[692,-3.8179],[708,-4.3547],[726,-4.5835],[747,-4.3882],[748,-4.3846],[770,-4.1745],[803,-3.1825],[821,-3.0335],[837,-3.2763],[887,-7.147],[895,-6.7574],[926,-0.7187],[927,-0.6734],[929,-0.6308]]
       }
       thigh_r_fk_ctrl.Rotation.Z {
        type=MovieSceneScriptingFloatChannel
        default=-0.1592
        thinned_from=98
        keys [[-2,6.8202],[0,6.8153],[11,6.6606],[14,6.5605],[21,6.3735],[50,7.4665],[57,8.0702],[62,8.8624],[83,12.6904],[90,13.5545],[97,14.1848],[110,14.7763],[146,11.7667],[185,9.7902],[203,10.5567],[216,10.9346],[235,11.2364],[285,10.0482],[309,7.4526],[323,6.8202],[325,6.8153],[336,6.6606],[339,6.5605],[346,6.3735],[375,7.4665],[382,8.0702],[387,8.8624],[408,12.6904],[415,13.5545],[422,14.1848],[435,14.7763],[471,11.7667],[510,9.7902],[528,10.5567],[541,10.9346],[560,11.2364],[610,10.0482],[634,7.4526],[676,-4.4439],[677,-4.4453],[699,-4.8906],[728,-3.7976],[778,-2.6155],[835,-4.4501],[837,-4.4439],[887,10.7047],[892,11.0381],[926,6.8242],[927,6.814],[929,6.8202]]
       }
       thigh_r_fk_ctrl.Scale.X {type=MovieSceneScriptingFloatChannel, default=1, const [1,751]}
       thigh_r_fk_ctrl.Scale.Y {type=MovieSceneScriptingFloatChannel, default=1, const [1,751]}
       thigh_r_fk_ctrl.Scale.Z {type=MovieSceneScriptingFloatChannel, default=1, const [1,751]}
       calf_r_fk_ctrl.Location.X {type=MovieSceneScriptingFloatChannel, default=0, thinned_from=754, keys [[-2,0],[929,0]]}
       calf_r_fk_ctrl.Location.Y {type=MovieSceneScriptingFloatChannel, default=0, thinned_from=754, keys [[-2,0],[929,0]]}
       calf_r_fk_ctrl.Location.Z {type=MovieSceneScriptingFloatChannel, default=0, thinned_from=754, keys [[-2,0],[929,0]]}
       calf_r_fk_ctrl.Rotation.X {
        type=MovieSceneScriptingFloatChannel
        default=0.0561
        thinned_from=754
        keys [[-2,0.8716],[-1,0.8548],[0,0.856],[10,0.8261],[13,0.8219],[18,0.8179],[22,0.8235],[67,0.8446],[76,0.838],[80,0.8278],[81,0.8233],[82,0.817],[84,0.7998],[86,0.7776],[89,0.7374],[94,0.6649],[96,0.6383],[98,0.6152],[99,0.6056],[100,0.5974],[101,0.5908],[102,0.5858],[103,0.5826],[104,0.5813],[105,0.5818],[106,0.584],[108,0.5921],[110,0.6041],[114,0.6351],[120,0.6863],[122,0.7007],[124,0.7118],[132,0.7342],[144,0.7472],[160,0.744],[190,0.6728],[197,0.6652],[212,0.6733],[217,0.6823],[222,0.6987],[226,0.7201],[238,0.8026],[241,0.8187],[244,0.8294],[246,0.833],[256,0.8298],[286,0.8803],[309,0.8833],[323,0.8716],[324,0.8548],[325,0.856],[335,0.8261],[338,0.8219],[343,0.8179],[347,0.8235],[392,0.8446],[401,0.838],[405,0.8278],[406,0.8233],[407,0.817],[409,0.7998],[411,0.7776],[414,0.7374],[419,0.6649],[421,0.6383],[423,0.6152],[424,0.6056],[425,0.5974],[426,0.5908],[427,0.5858],[428,0.5826],[429,0.5813],[430,0.5818],[431,0.584],[433,0.5921],[435,0.6041],[439,0.6351],[445,0.6863],[447,0.7007],[449,0.7118],[457,0.7342],[469,0.7472],[485,0.744],[515,0.6728],[522,0.6652],[537,0.6733],[542,0.6823],[547,0.6987],[551,0.7201],[563,0.8026],[566,0.8187],[569,0.8294],[571,0.833],[581,0.8298],[611,0.8803],[634,0.8833],[676,0.8964],[677,0.8796],[678,0.8809],[688,0.8509],[691,0.8468],[696,0.8427],[700,0.8484],[745,0.8694],[754,0.8629],[757,0.8559],[759,0.8481],[770,0.9043],[822,0.9031],[837,0.8964],[896,0.8801],[914,0.8782],[929,0.8716]]
       }
       calf_r_fk_ctrl.Rotation.Y {
        type=MovieSceneScriptingFloatChannel
        default=1.1982
        thinned_from=754
        keys [[-2,0.3414],[87,0.3434],[103,0.3521],[208,0.3469],[218,0.3541],[235,0.3799],[263,0.4359],[276,0.4509],[283,0.4527],[289,0.4479],[295,0.437],[309,0.3986],[323,0.3414],[412,0.3434],[428,0.3521],[533,0.3469],[543,0.3541],[560,0.3799],[588,0.4359],[601,0.4509],[608,0.4527],[614,0.4479],[620,0.437],[634,0.3986],[651,0.3314],[676,0.2694],[759,0.2701],[770,0.3799],[774,0.3781],[779,0.371],[785,0.3576],[803,0.3029],[822,0.2694],[837,0.2694],[855,0.3067],[887,0.3955],[895,0.3721],[914,0.3414],[929,0.3414]]
       }
       calf_r_fk_ctrl.Rotation.Z {
        type=MovieSceneScriptingFloatChannel
        default=0.2616
        thinned_from=512
        keys [[-2,9.3966],[1,9.3621],[12,7.9487],[14,7.7104],[53,3.6559],[55,3.6194],[57,3.6316],[59,3.6669],[61,3.717],[65,3.8564],[71,4.0992],[72,4.1544],[73,4.2383],[74,4.3489],[75,4.4811],[76,4.6328],[77,4.7977],[78,4.9767],[80,5.3632],[81,5.5484],[82,5.7079],[83,5.8444],[84,5.9579],[94,7.869],[96,8.2944],[98,8.6658],[100,8.9731],[102,9.217],[104,9.3916],[105,9.4522],[106,9.4902],[107,9.5077],[119,8.8069],[148,5.5849],[159,4.0594],[161,3.7626],[163,3.4416],[169,2.417],[171,2.0868],[172,1.9285],[173,1.7754],[174,1.6292],[175,1.4931],[176,1.3634],[177,1.2428],[178,1.1298],[179,1.0263],[180,0.9399],[181,0.8684],[182,0.8113],[183,0.768],[184,0.7394],[185,0.7281],[186,0.7274],[187,0.7443],[188,0.7752],[189,0.8164],[190,0.8767],[191,0.9464],[192,1.0326],[193,1.1297],[194,1.2416],[195,1.3649],[196,1.5034],[197,1.6506],[198,1.8148],[199,1.988],[200,2.1727],[201,2.3712],[207,3.6124],[209,4.0397],[227,8.0372],[229,8.4445],[233,9.1811],[235,9.4987],[237,9.7742],[239,10.0018],[241,10.1736],[243,10.2816],[245,10.3139],[246,10.2968],[247,10.2553],[260,9.1281],[274,7.784],[281,7.2219],[284,6.9981],[291,6.3801],[293,6.2319],[295,6.122],[296,6.0843],[297,6.0603],[302,6.0714],[309,6.1811],[335,7.7311],[340,7.6477],[350,7.1226],[378,3.6559],[380,3.6194],[382,3.6316],[384,3.6669],[386,3.717],[390,3.8564],[396,4.0992],[397,4.1544],[398,4.2383],[399,4.3489],[400,4.4811],[401,4.6328],[402,4.7977],[403,4.9767],[405,5.3632],[406,5.5484],[407,5.7079],[408,5.8444],[409,5.9579],[419,7.869],[421,8.2944],[423,8.6658],[425,8.9731],[427,9.217],[429,9.3916],[430,9.4522],[431,9.4902],[432,9.5077],[444,8.8069],[473,5.5849],[484,4.0594],[486,3.7626],[488,3.4416],[494,2.417],[496,2.0868],[497,1.9285],[498,1.7754],[499,1.6292],[500,1.4931],[501,1.3634],[502,1.2428],[503,1.1298],[504,1.0263],[505,0.9399],[506,0.8684],[507,0.8113],[508,0.768],[509,0.7394],[510,0.7281],[511,0.7274],[512,0.7443],[513,0.7752],[514,0.8164],[515,0.8767],[516,0.9464],[517,1.0326],[518,1.1297],[519,1.2416],[520,1.3649],[521,1.5034],[522,1.6506],[523,1.8148],[524,1.988],[525,2.1727],[526,2.3712],[532,3.6124],[534,4.0397],[552,8.0372],[554,8.4445],[558,9.1811],[560,9.4987],[562,9.7742],[564,10.0018],[566,10.1736],[568,10.2816],[570,10.3139],[571,10.2968],[572,10.2553],[585,9.1281],[599,7.784],[606,7.2219],[609,6.9981],[616,6.3801],[618,6.2319],[620,6.122],[621,6.0843],[622,6.0603],[627,6.0714],[634,6.1811],[676,16.4866],[690,15.7906],[692,15.6172],[698,15.572],[741,15.0084],[753,14.5377],[755,14.409],[757,14.2468],[759,14.0291],[770,13.7261],[786,13.8913],[803,13.7557],[837,14.0765],[855,12.6594],[887,13.6824],[889,13.8662],[891,13.9898],[893,14.0482],[895,14.025],[914,11.1563],[926,9.3869],[929,9.3966]]
       }
       calf_r_fk_ctrl.Scale.X {type=MovieSceneScriptingFloatChannel, default=1, const [1,754]}
       calf_r_fk_ctrl.Scale.Y {type=MovieSceneScriptingFloatChannel, default=1, const [1,754]}
       calf_r_fk_ctrl.Scale.Z {type=MovieSceneScriptingFloatChannel, default=1, const [1,754]}
       foot_r_fk_ctrl.Location.X {
        type=MovieSceneScriptingFloatChannel
        default=0
        thinned_from=746
        keys [[-2,0.0542],[-1,0.0301],[0,0.0181],[310,0.0181],[323,0.0542],[324,0.0301],[325,0.0181],[635,0.0181],[676,0.0542],[677,0.0301],[678,0.0181],[803,0.0181],[834,0.0534],[837,0.0542],[887,0.0181],[895,0.0181],[926,0.0534],[929,0.0542]]
       }
       foot_r_fk_ctrl.Location.Y {type=MovieSceneScriptingFloatChannel, default=-0.0329, thinned_from=746, keys [[-2,0],[929,0]]}
       foot_r_fk_ctrl.Location.Z {
        type=MovieSceneScriptingFloatChannel
        default=-0.9075
        thinned_from=746
        keys [[-2,-0.0001],[-1,0.0001],[323,-0.0001],[324,0.0001],[676,-0.0001],[677,0.0001],[929,-0.0001]]
       }
       foot_r_fk_ctrl.Rotation.X {
        type=MovieSceneScriptingFloatChannel
        default=-1.1648
        thinned_from=746
        keys [[-2,10.563],[-1,10.8048],[0,11.1295],[3,11.5737],[9,12.2926],[12,12.5886],[13,12.6609],[14,12.702],[16,12.7009],[17,12.6642],[19,12.5149],[22,12.1964],[28,11.4834],[30,11.3027],[31,11.2422],[32,11.2084],[33,11.2144],[34,11.2447],[35,11.2999],[37,11.4868],[39,11.7459],[41,12.0484],[47,13.0555],[50,13.4676],[52,13.6559],[53,13.7225],[55,13.7729],[58,13.7428],[61,13.6149],[65,13.3108],[82,11.505],[84,11.257],[85,11.1554],[86,11.1396],[87,11.2627],[88,11.4924],[89,11.8116],[90,12.1962],[93,13.4695],[94,13.8654],[95,14.2291],[96,14.5386],[97,14.776],[98,14.9292],[99,14.9776],[103,14.8696],[106,14.6977],[109,14.4301],[112,14.0606],[119,13.0071],[122,12.651],[126,12.2994],[129,12.1205],[132,12.0246],[135,12.0105],[148,12.225],[154,12.1721],[162,11.8439],[178,10.759],[201,8.7155],[210,8.1055],[214,7.8673],[225,6.8981],[230,6.526],[232,6.353],[234,6.1524],[235,6.0647],[236,6.0129],[237,5.9972],[238,6.02],[239,6.0676],[240,6.1465],[241,6.257],[242,6.3854],[243,6.5445],[244,6.7372],[245,6.9471],[246,7.1821],[247,7.4347],[249,7.9903],[255,9.7526],[258,10.5497],[260,11.0332],[262,11.4497],[263,11.6228],[267,12.2152],[279,14.3462],[284,15.0931],[287,15.4156],[297,16.0735],[309,16.5691],[310,16.5064],[323,10.563],[324,10.8048],[325,11.1295],[328,11.5737],[334,12.2926],[337,12.5886],[338,12.6609],[339,12.702],[341,12.7009],[342,12.6642],[344,12.5149],[347,12.1964],[353,11.4834],[355,11.3027],[356,11.2422],[357,11.2084],[358,11.2144],[359,11.2447],[360,11.2999],[362,11.4868],[364,11.7459],[366,12.0484],[372,13.0555],[375,13.4676],[377,13.6559],[378,13.7225],[380,13.7729],[383,13.7428],[386,13.6149],[390,13.3108],[407,11.505],[409,11.257],[410,11.1554],[411,11.1396],[412,11.2627],[413,11.4924],[414,11.8116],[415,12.1962],[418,13.4695],[419,13.8654],[420,14.2291],[421,14.5386],[422,14.776],[423,14.9292],[424,14.9776],[428,14.8696],[431,14.6977],[434,14.4301],[437,14.0606],[444,13.0071],[447,12.651],[451,12.2994],[454,12.1205],[457,12.0246],[460,12.0105],[473,12.225],[479,12.1721],[487,11.8439],[503,10.759],[526,8.7155],[535,8.1055],[539,7.8673],[550,6.8981],[555,6.526],[557,6.353],[559,6.1524],[560,6.0647],[561,6.0129],[562,5.9972],[563,6.02],[564,6.0676],[565,6.1465],[566,6.257],[567,6.3854],[568,6.5445],[569,6.7372],[570,6.9471],[571,7.1821],[572,7.4347],[574,7.9903],[580,9.7526],[583,10.5497],[585,11.0332],[587,11.4497],[588,11.6228],[592,12.2152],[604,14.3462],[609,15.0931],[612,15.4156],[622,16.0735],[634,16.5691],[635,16.5064],[676,-0.8205],[677,-0.5807],[678,-0.2563],[679,-0.1105],[680,0.0191],[681,0.1181],[682,0.1931],[683,0.2557],[687,0.3733],[688,0.3718],[689,0.3602],[690,0.3339],[691,0.2863],[692,0.2052],[693,0.093],[694,-0.0428],[695,-0.202],[696,-0.3895],[697,-0.589],[699,-1.0127],[701,-1.4491],[702,-1.6542],[703,-1.8506],[704,-2.0391],[705,-2.2071],[706,-2.3561],[707,-2.4782],[708,-2.5743],[709,-2.6364],[710,-2.668],[711,-2.6564],[712,-2.6169],[713,-2.5494],[714,-2.4481],[715,-2.3283],[716,-2.1851],[717,-2.0232],[718,-1.8479],[719,-1.6638],[721,-1.259],[723,-0.8379],[724,-0.6347],[725,-0.4301],[726,-0.2358],[727,-0.0523],[728,0.121],[729,0.2774],[730,0.4094],[731,0.5279],[732,0.6159],[733,0.6852],[734,0.7441],[735,0.7881],[736,0.8217],[737,0.8469],[739,0.865],[740,0.8556],[742,0.8223],[743,0.7907],[744,0.754],[745,0.7116],[746,0.6635],[747,0.6099],[748,0.5523],[749,0.4904],[750,0.4267],[751,0.3654],[752,0.3029],[754,0.1836],[755,0.125],[759,-0.1054],[770,1.422],[771,1.4206],[773,1.4065],[774,1.4037],[775,1.4091],[776,1.4176],[777,1.4326],[778,1.454],[780,1.5136],[781,1.5496],[783,1.6337],[785,1.7352],[786,1.7906],[788,1.9167],[789,1.9847],[791,2.1346],[792,2.2143],[794,2.3869],[796,2.5696],[799,2.866],[803,3.2841],[834,-0.6832],[837,-0.8225],[887,14.8363],[895,15.4801],[926,10.7023],[929,10.563]]
       }
       foot_r_fk_ctrl.Rotation.Y {
        type=MovieSceneScriptingFloatChannel
        default=0.1067
        thinned_from=526
        keys [[-2,-16.7726],[-1,-16.6569],[0,-16.467],[2,-16.2587],[4,-15.9645],[5,-15.789],[9,-14.9328],[13,-13.8908],[20,-11.8194],[23,-11.0317],[25,-10.5743],[26,-10.377],[27,-10.2033],[28,-10.0607],[29,-9.9457],[30,-9.8625],[31,-9.8114],[32,-9.7837],[33,-9.784],[34,-9.8077],[35,-9.8523],[37,-10.0013],[39,-10.2167],[41,-10.4825],[52,-12.2946],[59,-13.1914],[67,-14.0064],[78,-14.7716],[86,-15.2104],[88,-15.2236],[93,-15.1223],[95,-15.1407],[99,-15.3425],[101,-15.3712],[103,-15.3279],[112,-14.7611],[120,-14.4329],[147,-12.3603],[153,-12.053],[157,-11.9605],[163,-11.9817],[176,-12.3941],[201,-13.9548],[212,-14.3361],[216,-14.4341],[228,-14.8222],[272,-14.6818],[288,-14.5021],[304,-14.9578],[310,-14.5195],[323,-16.7726],[324,-16.6569],[325,-16.467],[327,-16.2587],[329,-15.9645],[330,-15.789],[334,-14.9328],[338,-13.8908],[345,-11.8194],[348,-11.0317],[350,-10.5743],[351,-10.377],[352,-10.2033],[353,-10.0607],[354,-9.9457],[355,-9.8625],[356,-9.8114],[357,-9.7837],[358,-9.784],[359,-9.8077],[360,-9.8523],[362,-10.0013],[364,-10.2167],[366,-10.4825],[377,-12.2946],[384,-13.1914],[392,-14.0064],[403,-14.7716],[411,-15.2104],[413,-15.2236],[418,-15.1223],[420,-15.1407],[424,-15.3425],[426,-15.3712],[428,-15.3279],[437,-14.7611],[445,-14.4329],[472,-12.3603],[478,-12.053],[482,-11.9605],[488,-11.9817],[501,-12.3941],[526,-13.9548],[537,-14.3361],[541,-14.4341],[553,-14.8222],[597,-14.6818],[613,-14.5021],[629,-14.9578],[635,-14.5195],[676,1.0006],[677,1.1146],[678,1.3045],[679,1.39],[680,1.4852],[681,1.5898],[682,1.7018],[683,1.8222],[687,2.3714],[688,2.5245],[690,2.8448],[691,3.0141],[693,3.3781],[696,3.9585],[698,4.3498],[701,4.8749],[703,5.1942],[704,5.3364],[705,5.4655],[706,5.5752],[707,5.6697],[708,5.7458],[709,5.7985],[710,5.8312],[711,5.839],[712,5.8264],[713,5.796],[714,5.7456],[715,5.6835],[717,5.5152],[719,5.3059],[722,4.9356],[726,4.4086],[728,4.1607],[729,4.0439],[730,3.9375],[731,3.8395],[733,3.6758],[736,3.4602],[737,3.394],[739,3.2744],[742,3.1192],[745,2.9936],[747,2.9239],[750,2.8426],[754,2.7549],[757,2.7053],[759,2.6767],[789,1.0155],[837,0.9989],[926,-16.7725],[929,-16.7726]]
       }
       foot_r_fk_ctrl.Rotation.Z {
        type=MovieSceneScriptingFloatChannel
        default=-0.1036
        thinned_from=251
        keys [[-2,30.4496],[0,30.4487],[30,18.5241],[92,29.3314],[96,29.1749],[98,28.6801],[108,27.4463],[112,26.7467],[115,26.0402],[117,25.4607],[118,25.1208],[119,24.7038],[120,24.1698],[155,18.1565],[160,17.8701],[162,17.6586],[164,17.3812],[167,16.8314],[174,15.3538],[175,15.1735],[177,14.9209],[178,14.8472],[179,14.8118],[181,14.8327],[182,14.8672],[184,15.0143],[186,15.2355],[189,15.7018],[191,16.0872],[228,24.1178],[230,24.1325],[231,24.3468],[260,25.4178],[262,25.4169],[325,18.5241],[349,17.8299],[424,28.3416],[442,25.902],[480,18.1565],[485,17.8701],[487,17.6586],[489,17.3812],[492,16.8314],[499,15.3538],[500,15.1735],[502,14.9209],[503,14.8472],[504,14.8118],[506,14.8327],[507,14.8672],[509,15.0143],[511,15.2355],[514,15.7018],[517,16.3033],[520,17.0178],[526,18.7009],[528,19.1908],[553,24.1178],[555,24.1325],[556,24.3468],[560,24.4525],[564,24.3158],[568,23.9419],[574,23.0357],[594,19.4553],[599,18.7531],[605,18.1552],[610,17.8385],[624,17.492],[628,17.5216],[630,17.6094],[633,17.8783],[676,23.0632],[770,23.3829],[776,25.135],[781,26.2959],[786,27.1411],[790,27.5766],[794,27.7488],[800,27.6083],[803,27.4161],[834,23.1461],[837,23.0758],[887,35.1281],[895,34.79],[926,30.5199],[929,30.4496]]
       }
       foot_r_fk_ctrl.Scale.X {type=MovieSceneScriptingFloatChannel, default=1, const [1,746]}
       foot_r_fk_ctrl.Scale.Y {type=MovieSceneScriptingFloatChannel, default=1, const [1,746]}
       foot_r_fk_ctrl.Scale.Z {type=MovieSceneScriptingFloatChannel, default=1, const [1,746]}
       ball_r_fk_ctrl.Location.X {type=MovieSceneScriptingFloatChannel, default=0, thinned_from=751, keys [[-2,0],[929,0]]}
       ball_r_fk_ctrl.Location.Y {type=MovieSceneScriptingFloatChannel, default=0, thinned_from=751, keys [[-2,0],[929,0]]}
       ball_r_fk_ctrl.Location.Z {type=MovieSceneScriptingFloatChannel, default=0, thinned_from=751, keys [[-2,0],[929,0]]}
       ball_r_fk_ctrl.Rotation.X {
        type=MovieSceneScriptingFloatChannel
        default=0
        thinned_from=751
        keys [[-2,0.1012],[-1,0.1025],[2,0.1026],[3,0.1015],[4,0.1015],[5,0.0986],[7,0.0986],[8,0.0933],[10,0.0933],[11,0.0863],[13,0.0863],[14,0.0768],[15,0.0793],[17,0.0793],[18,0.061],[20,0.061],[21,0.0472],[23,0.0472],[24,0.034],[26,0.034],[27,0.0231],[30,0.0231],[31,0.0237],[32,0.0237],[33,0.0146],[38,0.0146],[39,0.0295],[42,0.0295],[43,0.0444],[45,0.0444],[46,0.0592],[47,0.0577],[49,0.0577],[50,0.0729],[54,0.0729],[55,0.0866],[62,0.0866],[63,0.0871],[64,0.1052],[72,0.1052],[73,0.1016],[78,0.1016],[79,0.1026],[81,0.1026],[82,0.1033],[84,0.1033],[85,0.103],[86,0.103],[87,0.1062],[88,0.1103],[89,0.1168],[90,0.1238],[91,0.1301],[92,0.1301],[93,0.1451],[94,0.1451],[95,0.1591],[97,0.1591],[98,0.1722],[110,0.1722],[111,0.1597],[115,0.1597],[116,0.1477],[119,0.1477],[120,0.1352],[121,0.1352],[122,0.1293],[123,0.1293],[124,0.1235],[125,0.1235],[126,0.1188],[127,0.1192],[128,0.1138],[129,0.1138],[130,0.1096],[131,0.1096],[132,0.1064],[142,0.1064],[143,0.1068],[158,0.1068],[159,0.107],[169,0.107],[170,0.1051],[174,0.1051],[175,0.1049],[184,0.105],[185,0.1041],[190,0.1041],[206,0.1043],[207,0.1046],[208,0.1039],[215,0.1039],[217,0.103],[218,0.1034],[219,0.103],[220,0.1034],[222,0.1033],[224,0.1041],[225,0.104],[227,0.1048],[230,0.1049],[231,0.1053],[232,0.1053],[233,0.1057],[235,0.1057],[236,0.1072],[238,0.1072],[239,0.1155],[240,0.1155],[241,0.1251],[242,0.1251],[243,0.1378],[244,0.1378],[245,0.1506],[246,0.1506],[247,0.1678],[248,0.1678],[249,0.1761],[250,0.1863],[251,0.1863],[252,0.2022],[253,0.2022],[254,0.215],[255,0.2131],[256,0.2245],[257,0.2245],[258,0.2378],[260,0.2378],[261,0.2491],[266,0.2491],[267,0.2571],[270,0.2571],[271,0.2562],[274,0.2562],[275,0.2582],[277,0.2582],[278,0.2603],[280,0.2603],[281,0.2621],[283,0.2621],[284,0.2639],[285,0.2639],[286,0.2655],[287,0.2652],[288,0.2666],[290,0.2666],[291,0.2677],[293,0.2677],[294,0.2691],[296,0.2691],[297,0.2702],[301,0.2702],[302,0.2719],[309,0.2717],[323,0.1012],[324,0.1025],[327,0.1026],[328,0.1015],[329,0.1015],[330,0.0986],[332,0.0986],[333,0.0933],[335,0.0933],[336,0.0863],[338,0.0863],[339,0.0768],[340,0.0793],[342,0.0793],[343,0.061],[345,0.061],[346,0.0472],[348,0.0472],[349,0.034],[351,0.034],[352,0.0231],[355,0.0231],[356,0.0237],[357,0.0237],[358,0.0146],[363,0.0146],[364,0.0295],[367,0.0295],[368,0.0444],[370,0.0444],[371,0.0592],[372,0.0577],[374,0.0577],[375,0.0729],[379,0.0729],[380,0.0866],[387,0.0866],[388,0.0871],[389,0.1052],[397,0.1052],[398,0.1016],[403,0.1016],[404,0.1026],[406,0.1026],[407,0.1033],[409,0.1033],[410,0.103],[411,0.103],[412,0.1062],[413,0.1103],[414,0.1168],[415,0.1238],[416,0.1301],[417,0.1301],[418,0.1451],[419,0.1451],[420,0.1591],[422,0.1591],[423,0.1722],[435,0.1722],[436,0.1597],[440,0.1597],[441,0.1477],[444,0.1477],[445,0.1352],[446,0.1352],[447,0.1293],[448,0.1293],[449,0.1235],[450,0.1235],[451,0.1188],[452,0.1192],[453,0.1138],[454,0.1138],[455,0.1096],[456,0.1096],[457,0.1064],[467,0.1064],[468,0.1068],[483,0.1068],[484,0.107],[494,0.107],[495,0.1051],[499,0.1051],[500,0.1049],[509,0.105],[510,0.1041],[515,0.1041],[531,0.1043],[532,0.1046],[533,0.1039],[540,0.1039],[542,0.103],[543,0.1034],[544,0.103],[545,0.1034],[547,0.1033],[549,0.1041],[550,0.104],[552,0.1048],[555,0.1049],[556,0.1053],[557,0.1053],[558,0.1057],[560,0.1057],[561,0.1072],[563,0.1072],[564,0.1155],[565,0.1155],[566,0.1251],[567,0.1251],[568,0.1378],[569,0.1378],[570,0.1506],[571,0.1506],[572,0.1678],[573,0.1678],[574,0.1761],[575,0.1863],[576,0.1863],[577,0.2022],[578,0.2022],[579,0.215],[580,0.2131],[581,0.2245],[582,0.2245],[583,0.2378],[585,0.2378],[586,0.2491],[591,0.2491],[592,0.2571],[595,0.2571],[596,0.2562],[599,0.2562],[600,0.2582],[602,0.2582],[603,0.2603],[605,0.2603],[606,0.2621],[608,0.2621],[609,0.2639],[610,0.2639],[611,0.2655],[612,0.2652],[613,0.2666],[615,0.2666],[616,0.2677],[618,0.2677],[619,0.2691],[621,0.2691],[622,0.2702],[626,0.2702],[627,0.2719],[634,0.2717],[676,0.1012],[677,0.1025],[680,0.1026],[681,0.1015],[682,0.1015],[683,0.0986],[685,0.0986],[686,0.0933],[688,0.0933],[689,0.0863],[691,0.0863],[692,0.0768],[693,0.0793],[695,0.0793],[696,0.061],[698,0.061],[699,0.0472],[701,0.0472],[702,0.034],[704,0.034],[705,0.0231],[708,0.0231],[709,0.0237],[710,0.0237],[711,0.0146],[716,0.0146],[717,0.0295],[720,0.0295],[721,0.0444],[723,0.0444],[724,0.0592],[725,0.0577],[727,0.0577],[728,0.0729],[732,0.0729],[733,0.0866],[740,0.0866],[741,0.0871],[742,0.1052],[750,0.1052],[751,0.1016],[755,0.1016],[757,0.1026],[759,0.1026],[770,0.2639],[771,0.2655],[772,0.2652],[773,0.2666],[775,0.2666],[776,0.2677],[778,0.2677],[779,0.2691],[781,0.2691],[782,0.2702],[786,0.2702],[787,0.2719],[802,0.2717],[803,0.2714],[837,0.1012],[887,0.2717],[894,0.2717],[895,0.2714],[929,0.1012]]
       }
       ball_r_fk_ctrl.Rotation.Y {
        type=MovieSceneScriptingFloatChannel
        default=0
        thinned_from=751
        keys [[-2,-0.0922],[-1,-0.091],[2,-0.0911],[3,-0.0924],[4,-0.0924],[5,-0.0895],[7,-0.0895],[8,-0.0892],[10,-0.0892],[11,-0.0846],[13,-0.0846],[14,-0.0826],[15,-0.0838],[17,-0.0838],[18,-0.0712],[20,-0.0712],[21,-0.0746],[23,-0.0746],[24,-0.0671],[26,-0.0671],[27,-0.0676],[30,-0.0676],[32,-0.0678],[33,-0.0647],[38,-0.0647],[39,-0.0696],[42,-0.0696],[43,-0.0745],[45,-0.0745],[46,-0.0794],[47,-0.0791],[49,-0.0791],[50,-0.0822],[54,-0.0822],[55,-0.0868],[62,-0.0868],[63,-0.0886],[64,-0.093],[72,-0.093],[73,-0.0893],[78,-0.0893],[79,-0.0916],[81,-0.0916],[82,-0.0902],[84,-0.0902],[85,-0.0899],[86,-0.0899],[87,-0.0869],[88,-0.0847],[89,-0.0809],[90,-0.0775],[91,-0.0756],[92,-0.0756],[93,-0.07],[94,-0.07],[95,-0.0651],[97,-0.0651],[98,-0.062],[110,-0.062],[111,-0.0675],[115,-0.0675],[116,-0.0705],[119,-0.0705],[120,-0.0765],[121,-0.0765],[122,-0.078],[123,-0.078],[124,-0.0796],[125,-0.0796],[126,-0.0823],[127,-0.0815],[128,-0.0833],[129,-0.0833],[130,-0.0839],[131,-0.0839],[132,-0.0855],[142,-0.0855],[143,-0.0847],[158,-0.0847],[169,-0.0849],[174,-0.085],[175,-0.0846],[184,-0.0847],[185,-0.0844],[190,-0.0844],[207,-0.0842],[208,-0.085],[215,-0.085],[217,-0.0856],[220,-0.0868],[221,-0.0876],[222,-0.0875],[224,-0.0884],[225,-0.0883],[227,-0.089],[228,-0.089],[229,-0.0887],[230,-0.0887],[231,-0.089],[232,-0.089],[233,-0.0894],[235,-0.0894],[236,-0.092],[238,-0.092],[239,-0.1038],[240,-0.1038],[241,-0.1173],[242,-0.1173],[243,-0.1352],[244,-0.1352],[245,-0.1531],[246,-0.1531],[247,-0.1742],[248,-0.1742],[249,-0.189],[250,-0.2018],[251,-0.2018],[252,-0.2242],[253,-0.2242],[254,-0.2421],[255,-0.2403],[256,-0.2576],[257,-0.2576],[258,-0.2729],[260,-0.2729],[261,-0.2902],[266,-0.2902],[267,-0.2964],[270,-0.2964],[271,-0.2913],[274,-0.2913],[275,-0.2901],[277,-0.2901],[278,-0.2888],[280,-0.2888],[281,-0.2874],[283,-0.2874],[284,-0.2859],[285,-0.2859],[286,-0.2852],[287,-0.2853],[288,-0.284],[290,-0.284],[291,-0.2833],[293,-0.2833],[294,-0.282],[296,-0.282],[297,-0.2813],[301,-0.2813],[302,-0.2794],[309,-0.2796],[323,-0.0922],[324,-0.091],[327,-0.0911],[328,-0.0924],[329,-0.0924],[330,-0.0895],[332,-0.0895],[333,-0.0892],[335,-0.0892],[336,-0.0846],[338,-0.0846],[339,-0.0826],[340,-0.0838],[342,-0.0838],[343,-0.0712],[345,-0.0712],[346,-0.0746],[348,-0.0746],[349,-0.0671],[351,-0.0671],[352,-0.0676],[355,-0.0676],[357,-0.0678],[358,-0.0647],[363,-0.0647],[364,-0.0696],[367,-0.0696],[368,-0.0745],[370,-0.0745],[371,-0.0794],[372,-0.0791],[374,-0.0791],[375,-0.0822],[379,-0.0822],[380,-0.0868],[387,-0.0868],[388,-0.0886],[389,-0.093],[397,-0.093],[398,-0.0893],[403,-0.0893],[404,-0.0916],[406,-0.0916],[407,-0.0902],[409,-0.0902],[410,-0.0899],[411,-0.0899],[412,-0.0869],[413,-0.0847],[414,-0.0809],[415,-0.0775],[416,-0.0756],[417,-0.0756],[418,-0.07],[419,-0.07],[420,-0.0651],[422,-0.0651],[423,-0.062],[435,-0.062],[436,-0.0675],[440,-0.0675],[441,-0.0705],[444,-0.0705],[445,-0.0765],[446,-0.0765],[447,-0.078],[448,-0.078],[449,-0.0796],[450,-0.0796],[451,-0.0823],[452,-0.0815],[453,-0.0833],[454,-0.0833],[455,-0.0839],[456,-0.0839],[457,-0.0855],[467,-0.0855],[468,-0.0847],[483,-0.0847],[494,-0.0849],[499,-0.085],[500,-0.0846],[509,-0.0847],[510,-0.0844],[515,-0.0844],[532,-0.0842],[533,-0.085],[540,-0.085],[542,-0.0856],[545,-0.0868],[546,-0.0876],[547,-0.0875],[549,-0.0884],[550,-0.0883],[552,-0.089],[553,-0.089],[554,-0.0887],[555,-0.0887],[556,-0.089],[557,-0.089],[558,-0.0894],[560,-0.0894],[561,-0.092],[563,-0.092],[564,-0.1038],[565,-0.1038],[566,-0.1173],[567,-0.1173],[568,-0.1352],[569,-0.1352],[570,-0.1531],[571,-0.1531],[572,-0.1742],[573,-0.1742],[574,-0.189],[575,-0.2018],[576,-0.2018],[577,-0.2242],[578,-0.2242],[579,-0.2421],[580,-0.2403],[581,-0.2576],[582,-0.2576],[583,-0.2729],[585,-0.2729],[586,-0.2902],[591,-0.2902],[592,-0.2964],[595,-0.2964],[596,-0.2913],[599,-0.2913],[600,-0.2901],[602,-0.2901],[603,-0.2888],[605,-0.2888],[606,-0.2874],[608,-0.2874],[609,-0.2859],[610,-0.2859],[611,-0.2852],[612,-0.2853],[613,-0.284],[615,-0.284],[616,-0.2833],[618,-0.2833],[619,-0.282],[621,-0.282],[622,-0.2813],[626,-0.2813],[627,-0.2794],[634,-0.2796],[676,-0.0922],[677,-0.091],[680,-0.0911],[681,-0.0924],[682,-0.0924],[683,-0.0895],[685,-0.0895],[686,-0.0892],[688,-0.0892],[689,-0.0846],[691,-0.0846],[692,-0.0826],[693,-0.0838],[695,-0.0838],[696,-0.0712],[698,-0.0712],[699,-0.0746],[701,-0.0746],[702,-0.0671],[704,-0.0671],[705,-0.0676],[708,-0.0676],[710,-0.0678],[711,-0.0647],[716,-0.0647],[717,-0.0696],[720,-0.0696],[721,-0.0745],[723,-0.0745],[724,-0.0794],[725,-0.0791],[727,-0.0791],[728,-0.0822],[732,-0.0822],[733,-0.0868],[740,-0.0868],[741,-0.0886],[742,-0.093],[750,-0.093],[751,-0.0893],[755,-0.0893],[757,-0.0916],[759,-0.0916],[770,-0.2859],[771,-0.2852],[772,-0.2853],[773,-0.284],[775,-0.284],[776,-0.2833],[778,-0.2833],[779,-0.282],[781,-0.282],[782,-0.2813],[786,-0.2813],[787,-0.2794],[802,-0.2796],[803,-0.2806],[837,-0.0922],[887,-0.2796],[894,-0.2796],[895,-0.2806],[929,-0.0922]]
       }
       ball_r_fk_ctrl.Rotation.Z {
        type=MovieSceneScriptingFloatChannel
        default=0
        thinned_from=751
        keys [[-2,0.56],[-1,0.5431],[0,0.5431],[1,0.5262],[2,0.5262],[3,0.5093],[4,0.5093],[5,0.4924],[7,0.4924],[8,0.4755],[10,0.4755],[11,0.4585],[13,0.4586],[14,0.4416],[17,0.4425],[18,0.4295],[20,0.4295],[21,0.4231],[26,0.423],[27,0.4294],[30,0.4294],[31,0.427],[32,0.427],[33,0.4419],[38,0.4419],[39,0.451],[42,0.451],[43,0.463],[45,0.463],[46,0.4691],[47,0.468],[49,0.468],[50,0.4733],[54,0.4733],[55,0.4786],[62,0.4786],[63,0.4804],[64,0.4751],[72,0.4751],[73,0.4642],[78,0.4642],[79,0.4631],[81,0.4631],[82,0.4527],[84,0.4527],[85,0.4319],[86,0.4319],[87,0.406],[90,0.3437],[91,0.3282],[92,0.3282],[93,0.3075],[94,0.3075],[95,0.2967],[97,0.2967],[98,0.2873],[110,0.2873],[111,0.2881],[119,0.288],[120,0.2774],[121,0.2774],[122,0.2562],[123,0.2562],[124,0.235],[125,0.235],[126,0.2138],[127,0.2179],[128,0.1944],[129,0.1944],[130,0.1787],[131,0.1787],[132,0.163],[158,0.163],[159,0.1603],[169,0.1603],[170,0.1467],[174,0.1467],[175,0.1508],[176,0.1291],[184,0.1291],[185,0.1183],[190,0.1183],[191,0.1196],[206,0.1196],[207,0.1235],[208,0.104],[215,0.104],[222,0.2402],[223,0.2606],[226,0.3127],[227,0.3214],[228,0.3214],[229,0.3387],[230,0.3387],[231,0.3561],[232,0.3561],[233,0.3735],[235,0.3735],[236,0.3909],[238,0.3909],[239,0.3995],[240,0.3995],[241,0.4044],[242,0.4044],[243,0.4069],[244,0.4069],[245,0.4078],[248,0.408],[250,0.4066],[253,0.406],[257,0.4065],[258,0.4102],[260,0.4101],[261,0.4138],[266,0.4138],[267,0.4278],[270,0.4278],[271,0.4374],[274,0.4374],[275,0.4549],[277,0.4549],[278,0.4725],[280,0.4725],[281,0.49],[283,0.49],[284,0.5075],[285,0.5075],[286,0.5192],[287,0.521],[288,0.5344],[290,0.5344],[291,0.5479],[293,0.5479],[294,0.5614],[296,0.5614],[297,0.5703],[301,0.5703],[302,0.5838],[303,0.5856],[309,0.5856],[323,0.56],[324,0.5431],[325,0.5431],[326,0.5262],[327,0.5262],[328,0.5093],[329,0.5093],[330,0.4924],[332,0.4924],[333,0.4755],[335,0.4755],[336,0.4585],[338,0.4586],[339,0.4416],[342,0.4425],[343,0.4295],[345,0.4295],[346,0.4231],[351,0.423],[352,0.4294],[355,0.4294],[356,0.427],[357,0.427],[358,0.4419],[363,0.4419],[364,0.451],[367,0.451],[368,0.463],[370,0.463],[371,0.4691],[372,0.468],[374,0.468],[375,0.4733],[379,0.4733],[380,0.4786],[387,0.4786],[388,0.4804],[389,0.4751],[397,0.4751],[398,0.4642],[403,0.4642],[404,0.4631],[406,0.4631],[407,0.4527],[409,0.4527],[410,0.4319],[411,0.4319],[412,0.406],[415,0.3437],[416,0.3282],[417,0.3282],[418,0.3075],[419,0.3075],[420,0.2967],[422,0.2967],[423,0.2873],[435,0.2873],[436,0.2881],[444,0.288],[445,0.2774],[446,0.2774],[447,0.2562],[448,0.2562],[449,0.235],[450,0.235],[451,0.2138],[452,0.2179],[453,0.1944],[454,0.1944],[455,0.1787],[456,0.1787],[457,0.163],[483,0.163],[484,0.1603],[494,0.1603],[495,0.1467],[499,0.1467],[500,0.1508],[501,0.1291],[509,0.1291],[510,0.1183],[515,0.1183],[516,0.1196],[531,0.1196],[532,0.1235],[533,0.104],[540,0.104],[547,0.2402],[548,0.2606],[551,0.3127],[552,0.3214],[553,0.3214],[554,0.3387],[555,0.3387],[556,0.3561],[557,0.3561],[558,0.3735],[560,0.3735],[561,0.3909],[563,0.3909],[564,0.3995],[565,0.3995],[566,0.4044],[567,0.4044],[568,0.4069],[569,0.4069],[570,0.4078],[573,0.408],[575,0.4066],[578,0.406],[582,0.4065],[583,0.4102],[585,0.4101],[586,0.4138],[591,0.4138],[592,0.4278],[595,0.4278],[596,0.4374],[599,0.4374],[600,0.4549],[602,0.4549],[603,0.4725],[605,0.4725],[606,0.49],[608,0.49],[609,0.5075],[610,0.5075],[611,0.5192],[612,0.521],[613,0.5344],[615,0.5344],[616,0.5479],[618,0.5479],[619,0.5614],[621,0.5614],[622,0.5703],[626,0.5703],[627,0.5838],[628,0.5856],[634,0.5856],[676,0.56],[677,0.5431],[678,0.5431],[679,0.5262],[680,0.5262],[681,0.5093],[682,0.5093],[683,0.4924],[685,0.4924],[686,0.4755],[688,0.4755],[689,0.4585],[691,0.4586],[692,0.4416],[695,0.4425],[696,0.4295],[698,0.4295],[699,0.4231],[704,0.423],[705,0.4294],[708,0.4294],[709,0.427],[710,0.427],[711,0.4419],[716,0.4419],[717,0.451],[720,0.451],[721,0.463],[723,0.463],[724,0.4691],[725,0.468],[727,0.468],[728,0.4733],[732,0.4733],[733,0.4786],[740,0.4786],[741,0.4804],[742,0.4751],[750,0.4751],[751,0.4642],[755,0.4642],[757,0.4631],[759,0.4631],[770,0.5075],[771,0.5192],[772,0.521],[773,0.5344],[775,0.5344],[776,0.5479],[778,0.5479],[779,0.5614],[781,0.5614],[782,0.5703],[786,0.5703],[787,0.5838],[788,0.5856],[802,0.5857],[803,0.5699],[837,0.56],[894,0.5857],[895,0.5699],[929,0.56]]
       }
       ball_r_fk_ctrl.Scale.X {type=MovieSceneScriptingFloatChannel, default=1, const [1,751]}
       ball_r_fk_ctrl.Scale.Y {type=MovieSceneScriptingFloatChannel, default=1, const [1,751]}
       ball_r_fk_ctrl.Scale.Z {type=MovieSceneScriptingFloatChannel, default=1, const [1,751]}
       foot_r_ik_ctrl.Location.X {
        type=MovieSceneScriptingFloatChannel
        default=0
        thinned_from=753
        keys [[-2,-9.9153],[-1,-9.5448],[0,-9.5828],[1,-9.5343],[2,-9.5058],[4,-9.5089],[13,-9.8653],[15,-9.9105],[25,-9.922],[51,-10.6444],[75,-11.4744],[81,-11.6125],[83,-11.7111],[97,-12.7342],[99,-12.8332],[101,-12.8458],[103,-12.7865],[107,-12.5053],[112,-12.0931],[121,-11.5419],[150,-9.2857],[155,-8.9735],[161,-8.7178],[168,-8.5625],[182,-8.4863],[196,-8.4774],[206,-8.3052],[213,-8.0995],[216,-7.9675],[218,-7.835],[221,-7.5905],[222,-7.5295],[224,-7.4577],[226,-7.446],[227,-7.4586],[229,-7.5311],[231,-7.6525],[233,-7.8089],[235,-8.0074],[237,-8.26],[239,-8.5662],[241,-8.9227],[255,-11.5356],[258,-12.0254],[261,-12.4177],[264,-12.7215],[268,-13.01],[272,-13.1869],[278,-13.286],[292,-13.1865],[301,-12.8627],[309,-12.4215],[323,-9.9153],[324,-9.5448],[325,-9.5828],[326,-9.5343],[327,-9.5058],[329,-9.5089],[338,-9.8653],[340,-9.9105],[350,-9.922],[376,-10.6444],[400,-11.4744],[406,-11.6125],[408,-11.7111],[422,-12.7342],[424,-12.8332],[426,-12.8458],[428,-12.7865],[432,-12.5053],[437,-12.0931],[446,-11.5419],[475,-9.2857],[480,-8.9735],[486,-8.7178],[493,-8.5625],[507,-8.4863],[521,-8.4774],[531,-8.3052],[538,-8.0995],[541,-7.9675],[543,-7.835],[546,-7.5905],[547,-7.5295],[549,-7.4577],[551,-7.446],[552,-7.4586],[554,-7.5311],[556,-7.6525],[558,-7.8089],[560,-8.0074],[562,-8.26],[564,-8.5662],[566,-8.9227],[580,-11.5356],[583,-12.0254],[586,-12.4177],[589,-12.7215],[593,-13.01],[597,-13.1869],[603,-13.286],[617,-13.1865],[626,-12.8627],[634,-12.4215],[676,-9.9153],[677,-9.5448],[678,-9.5828],[679,-9.5343],[680,-9.5058],[682,-9.5089],[691,-9.8653],[693,-9.9105],[703,-9.922],[729,-10.6444],[753,-11.4744],[759,-11.6125],[770,-13.2641],[778,-13.1616],[786,-12.8627],[803,-11.8406],[834,-9.9141],[837,-9.9153],[887,-12.3595],[895,-11.8406],[926,-9.9141],[929,-9.9153]]
       }
       foot_r_ik_ctrl.Location.Y {
        type=MovieSceneScriptingFloatChannel
        default=0
        thinned_from=753
        keys [[-2,-1.1277],[-1,-0.5666],[0,-0.4246],[1,-0.2277],[2,-0.0461],[3,0.1212],[4,0.2808],[5,0.4248],[6,0.5503],[7,0.6569],[8,0.7421],[9,0.8084],[10,0.8584],[11,0.8798],[12,0.8926],[13,0.8908],[14,0.8722],[15,0.8989],[16,0.9325],[18,1.0191],[20,1.0803],[21,1.1092],[22,1.1307],[23,1.1089],[24,1.0802],[25,1.0452],[26,0.9953],[27,0.9423],[28,0.8791],[29,0.8201],[33,0.5483],[34,0.4745],[35,0.4103],[36,0.3418],[37,0.2757],[39,0.1505],[40,0.1],[41,0.0453],[43,-0.0385],[44,-0.0722],[45,-0.0983],[46,-0.1172],[47,-0.131],[48,-0.132],[49,-0.1327],[50,-0.1226],[51,-0.1009],[52,-0.0725],[53,-0.0385],[54,0.0134],[55,0.0738],[56,0.1522],[57,0.241],[58,0.3329],[59,0.4445],[60,0.5614],[61,0.6909],[62,0.8263],[63,0.9675],[64,1.1143],[66,1.4255],[67,1.5836],[71,2.2424],[73,2.5633],[75,2.8991],[76,3.0748],[77,3.2647],[78,3.4623],[79,3.6777],[80,3.903],[81,4.1795],[82,4.5144],[83,4.8963],[84,5.3145],[85,5.7489],[88,7.1092],[90,7.9877],[91,8.4052],[92,8.8028],[93,9.1757],[94,9.5173],[95,9.8206],[96,10.0847],[97,10.3033],[98,10.4329],[99,10.4507],[100,10.3396],[101,10.1171],[102,9.7885],[103,9.3578],[104,8.8254],[105,8.2091],[106,7.5206],[107,6.7965],[108,6.0529],[110,4.5295],[111,3.8001],[112,3.1047],[113,2.4745],[115,1.283],[116,0.6844],[117,0.095],[118,-0.4827],[119,-1.0425],[120,-1.588],[121,-2.099],[122,-2.5814],[123,-3.0188],[124,-3.4104],[125,-3.7569],[126,-4.0651],[127,-4.3529],[128,-4.6294],[129,-4.8781],[130,-5.1077],[131,-5.3085],[132,-5.4824],[133,-5.6321],[134,-5.7567],[135,-5.8527],[136,-5.917],[137,-5.96],[139,-5.9789],[141,-5.9306],[143,-5.8222],[145,-5.6783],[148,-5.4019],[153,-4.9021],[159,-4.3861],[160,-4.2937],[162,-4.0638],[163,-3.9357],[164,-3.7884],[165,-3.6508],[169,-3.0458],[173,-2.4273],[174,-2.2779],[176,-1.9892],[178,-1.716],[179,-1.5842],[180,-1.4557],[181,-1.3348],[182,-1.2171],[183,-1.1033],[184,-0.999],[185,-0.9009],[186,-0.8117],[187,-0.7318],[188,-0.6592],[189,-0.588],[190,-0.5315],[191,-0.4793],[193,-0.3907],[194,-0.3618],[195,-0.3272],[197,-0.2877],[198,-0.2744],[200,-0.2648],[201,-0.2751],[202,-0.2956],[203,-0.3155],[204,-0.343],[205,-0.3814],[206,-0.4268],[207,-0.4909],[208,-0.563],[209,-0.6616],[210,-0.7733],[211,-0.9109],[212,-1.0734],[213,-1.2584],[214,-1.4701],[215,-1.7065],[216,-1.9764],[217,-2.2908],[218,-2.6461],[219,-3.0438],[221,-3.9018],[222,-4.324],[223,-4.7362],[225,-5.5327],[227,-6.2949],[229,-6.9973],[230,-7.3177],[231,-7.6127],[232,-7.8868],[233,-8.1419],[234,-8.3655],[235,-8.5627],[236,-8.7282],[237,-8.8738],[238,-8.9873],[239,-9.0742],[240,-9.1257],[241,-9.1351],[242,-9.1002],[245,-8.9248],[248,-8.769],[249,-8.6946],[250,-8.6404],[252,-8.58],[253,-8.5689],[255,-8.6087],[259,-8.8213],[261,-8.971],[263,-9.1632],[275,-10.5107],[278,-10.7779],[280,-10.8951],[282,-10.9483],[283,-10.9496],[285,-10.8901],[287,-10.7433],[288,-10.6391],[290,-10.367],[291,-10.2015],[293,-9.8098],[294,-9.5776],[295,-9.3213],[296,-9.0447],[297,-8.748],[298,-8.4337],[300,-7.7504],[301,-7.39],[304,-6.2351],[308,-4.5988],[309,-4.1826],[323,-1.1277],[324,-0.5666],[325,-0.4246],[326,-0.2277],[327,-0.0461],[328,0.1212],[329,0.2808],[330,0.4248],[331,0.5503],[332,0.6569],[333,0.7421],[334,0.8084],[335,0.8584],[336,0.8798],[337,0.8926],[338,0.8908],[339,0.8722],[340,0.8989],[341,0.9325],[343,1.0191],[345,1.0803],[346,1.1092],[347,1.1307],[348,1.1089],[349,1.0802],[350,1.0452],[351,0.9953],[352,0.9423],[353,0.8791],[354,0.8201],[358,0.5483],[359,0.4745],[360,0.4103],[361,0.3418],[362,0.2757],[364,0.1505],[365,0.1],[366,0.0453],[368,-0.0385],[369,-0.0722],[370,-0.0983],[371,-0.1172],[372,-0.131],[373,-0.132],[374,-0.1327],[375,-0.1226],[376,-0.1009],[377,-0.0725],[378,-0.0385],[379,0.0134],[380,0.0738],[381,0.1522],[382,0.241],[383,0.3329],[384,0.4445],[385,0.5614],[386,0.6909],[387,0.8263],[388,0.9675],[389,1.1143],[391,1.4255],[392,1.5836],[396,2.2424],[398,2.5633],[400,2.8991],[401,3.0748],[402,3.2647],[403,3.4623],[404,3.6777],[405,3.903],[406,4.1795],[407,4.5144],[408,4.8963],[409,5.3145],[410,5.7489],[413,7.1092],[415,7.9877],[416,8.4052],[417,8.8028],[418,9.1757],[419,9.5173],[420,9.8206],[421,10.0847],[422,10.3033],[423,10.4329],[424,10.4507],[425,10.3396],[426,10.1171],[427,9.7885],[428,9.3578],[429,8.8254],[430,8.2091],[431,7.5206],[432,6.7965],[433,6.0529],[435,4.5295],[436,3.8001],[437,3.1047],[438,2.4745],[440,1.283],[441,0.6844],[442,0.095],[443,-0.4827],[444,-1.0425],[445,-1.588],[446,-2.099],[447,-2.5814],[448,-3.0188],[449,-3.4104],[450,-3.7569],[451,-4.0651],[452,-4.3529],[453,-4.6294],[454,-4.8781],[455,-5.1077],[456,-5.3085],[457,-5.4824],[458,-5.6321],[459,-5.7567],[460,-5.8527],[461,-5.917],[462,-5.96],[464,-5.9789],[466,-5.9306],[468,-5.8222],[470,-5.6783],[473,-5.4019],[478,-4.9021],[484,-4.3861],[485,-4.2937],[487,-4.0638],[488,-3.9357],[489,-3.7884],[490,-3.6508],[494,-3.0458],[498,-2.4273],[499,-2.2779],[501,-1.9892],[503,-1.716],[504,-1.5842],[505,-1.4557],[506,-1.3348],[507,-1.2171],[508,-1.1033],[509,-0.999],[510,-0.9009],[511,-0.8117],[512,-0.7318],[513,-0.6592],[514,-0.588],[515,-0.5315],[516,-0.4793],[518,-0.3907],[519,-0.3618],[520,-0.3272],[522,-0.2877],[523,-0.2744],[525,-0.2648],[526,-0.2751],[527,-0.2956],[528,-0.3155],[529,-0.343],[530,-0.3814],[531,-0.4268],[532,-0.4909],[533,-0.563],[534,-0.6616],[535,-0.7733],[536,-0.9109],[537,-1.0734],[538,-1.2584],[539,-1.4701],[540,-1.7065],[541,-1.9764],[542,-2.2908],[543,-2.6461],[544,-3.0438],[546,-3.9018],[547,-4.324],[548,-4.7362],[550,-5.5327],[552,-6.2949],[554,-6.9973],[555,-7.3177],[556,-7.6127],[557,-7.8868],[558,-8.1419],[559,-8.3655],[560,-8.5627],[561,-8.7282],[562,-8.8738],[563,-8.9873],[564,-9.0742],[565,-9.1257],[566,-9.1351],[567,-9.1002],[570,-8.9248],[573,-8.769],[574,-8.6946],[575,-8.6404],[577,-8.58],[578,-8.5689],[580,-8.6087],[584,-8.8213],[586,-8.971],[588,-9.1632],[600,-10.5107],[603,-10.7779],[605,-10.8951],[607,-10.9483],[608,-10.9496],[610,-10.8901],[612,-10.7433],[613,-10.6391],[615,-10.367],[616,-10.2015],[618,-9.8098],[619,-9.5776],[620,-9.3213],[621,-9.0447],[622,-8.748],[623,-8.4337],[625,-7.7504],[626,-7.39],[629,-6.2351],[633,-4.5988],[634,-4.1826],[676,-1.1277],[677,-0.5666],[678,-0.4246],[679,-0.2277],[680,-0.0461],[681,0.1212],[682,0.2808],[683,0.4248],[684,0.5503],[685,0.6569],[686,0.7421],[687,0.8084],[688,0.8584],[689,0.8798],[690,0.8926],[691,0.8908],[692,0.8722],[693,0.8989],[694,0.9325],[696,1.0191],[698,1.0803],[699,1.1092],[700,1.1307],[701,1.1089],[702,1.0802],[703,1.0452],[704,0.9953],[705,0.9423],[706,0.8791],[707,0.8201],[711,0.5483],[712,0.4745],[713,0.4103],[714,0.3418],[715,0.2757],[717,0.1505],[718,0.1],[719,0.0453],[721,-0.0385],[722,-0.0722],[723,-0.0983],[724,-0.1172],[725,-0.131],[726,-0.132],[727,-0.1327],[728,-0.1226],[729,-0.1009],[730,-0.0725],[731,-0.0385],[732,0.0134],[733,0.0738],[734,0.1522],[735,0.241],[736,0.3329],[737,0.4445],[738,0.5614],[739,0.6909],[740,0.8263],[741,0.9675],[742,1.1143],[744,1.4255],[745,1.5836],[749,2.2424],[751,2.5633],[753,2.8991],[754,3.0748],[755,3.2647],[757,3.6777],[759,4.1795],[770,-10.8901],[772,-10.7433],[773,-10.6391],[775,-10.367],[776,-10.2015],[778,-9.8098],[779,-9.5776],[780,-9.3213],[781,-9.0447],[782,-8.748],[783,-8.4337],[785,-7.7504],[786,-7.39],[789,-6.2351],[793,-4.5988],[795,-3.7683],[797,-2.9525],[799,-2.1499],[800,-1.7581],[801,-1.3746],[802,-0.9962],[803,-0.625],[834,-1.0318],[837,-1.1277],[887,-3.7683],[889,-2.9525],[891,-2.1499],[892,-1.7581],[893,-1.3746],[894,-0.9962],[895,-0.625],[926,-1.0318],[929,-1.1277]]
       }
       foot_r_ik_ctrl.Location.Z {
        type=MovieSceneScriptingFloatChannel
        default=0
        thinned_from=753
        keys [[-2,40.0546],[-1,40.2034],[0,40.0434],[2,39.8691],[5,39.3029],[18,35.6187],[26,33.989],[29,33.6151],[33,33.5485],[37,33.764],[60,36.3858],[67,36.7559],[72,36.5256],[77,35.8828],[88,33.9591],[92,33.5917],[95,33.5726],[105,34.3918],[108,34.8764],[111,35.653],[117,37.8593],[124,40.6186],[128,41.8038],[132,42.6546],[136,43.1385],[140,43.2104],[183,37.5294],[195,36.7852],[201,36.7887],[209,37.4306],[231,41.3595],[236,42.3633],[239,42.7331],[245,42.7856],[301,40.8352],[309,39.9832],[323,40.0546],[324,40.2034],[325,40.0434],[327,39.8691],[330,39.3029],[343,35.6187],[351,33.989],[354,33.6151],[358,33.5485],[362,33.764],[385,36.3858],[392,36.7559],[397,36.5256],[402,35.8828],[413,33.9591],[417,33.5917],[420,33.5726],[430,34.3918],[433,34.8764],[436,35.653],[442,37.8593],[449,40.6186],[453,41.8038],[457,42.6546],[461,43.1385],[465,43.2104],[508,37.5294],[520,36.7852],[526,36.7887],[534,37.4306],[556,41.3595],[561,42.3633],[564,42.7331],[570,42.7856],[626,40.8352],[634,39.9832],[676,40.0546],[677,40.2034],[678,40.0434],[680,39.8691],[683,39.3029],[696,35.6187],[704,33.989],[707,33.6151],[711,33.5485],[715,33.764],[738,36.3858],[745,36.7559],[750,36.5256],[755,35.8828],[759,35.1729],[770,41.3295],[786,40.8352],[803,38.762],[834,40.041],[887,39.8482],[895,38.762],[926,40.041],[929,40.0546]]
       }
       foot_r_ik_ctrl.Rotation.X {
        type=MovieSceneScriptingFloatChannel
        default=0
        thinned_from=753
        keys [[-2,57.2288],[-1,56.0251],[0,54.6238],[2,52.8675],[7,49.5273],[10,47.8502],[12,46.9594],[16,45.6571],[18,45.1811],[21,44.855],[25,44.8621],[33,45.8862],[49,48.7987],[55,49.2351],[62,48.8878],[73,47.0836],[78,45.7947],[80,45.0591],[81,44.5729],[82,43.9552],[83,43.2278],[84,42.233],[85,40.8682],[86,39.2173],[87,37.3525],[90,31.3462],[91,29.47],[92,27.7923],[93,26.3946],[94,25.2175],[95,24.1649],[96,23.2326],[97,22.4421],[98,21.8299],[99,21.4316],[100,21.2402],[101,21.2438],[102,21.4377],[103,21.8172],[104,22.3884],[105,23.1409],[106,24.0461],[107,25.058],[108,26.1478],[112,30.8568],[116,35.2491],[118,37.2531],[119,38.1112],[120,38.818],[121,39.3583],[122,39.7639],[123,40.0368],[124,40.1718],[130,40.0048],[131,40.0707],[132,40.2225],[133,40.4679],[138,42.3749],[142,43.5279],[149,44.8485],[157,45.6217],[160,45.6128],[162,45.366],[165,44.6467],[168,43.5678],[173,41.1797],[189,32.8046],[195,30.0827],[199,28.5708],[208,25.4487],[210,24.9151],[211,24.7199],[212,24.5824],[213,24.515],[214,24.6557],[215,25.101],[216,25.8267],[217,26.8042],[218,27.993],[219,29.3657],[220,30.8716],[224,37.2221],[225,38.6955],[226,40.0357],[227,41.1884],[230,44.161],[233,46.8051],[236,49.1106],[238,50.3806],[240,51.3555],[242,51.978],[245,52.4162],[247,52.4496],[253,51.9955],[270,53.6207],[290,58.9183],[295,59.6551],[301,59.7424],[307,59.0183],[309,58.5922],[323,57.2288],[324,56.0251],[325,54.6238],[327,52.8675],[332,49.5273],[335,47.8502],[337,46.9594],[341,45.6571],[343,45.1811],[346,44.855],[350,44.8621],[358,45.8862],[374,48.7987],[380,49.2351],[387,48.8878],[398,47.0836],[403,45.7947],[405,45.0591],[406,44.5729],[407,43.9552],[408,43.2278],[409,42.233],[410,40.8682],[411,39.2173],[412,37.3525],[415,31.3462],[416,29.47],[417,27.7923],[418,26.3946],[419,25.2175],[420,24.1649],[421,23.2326],[422,22.4421],[423,21.8299],[424,21.4316],[425,21.2402],[426,21.2438],[427,21.4377],[428,21.8172],[429,22.3884],[430,23.1409],[431,24.0461],[432,25.058],[433,26.1478],[437,30.8568],[441,35.2491],[443,37.2531],[444,38.1112],[445,38.818],[446,39.3583],[447,39.7639],[448,40.0368],[449,40.1718],[455,40.0048],[456,40.0707],[457,40.2225],[458,40.4679],[463,42.3749],[467,43.5279],[474,44.8485],[482,45.6217],[485,45.6128],[487,45.366],[490,44.6467],[493,43.5678],[498,41.1797],[514,32.8046],[520,30.0827],[524,28.5708],[533,25.4487],[535,24.9151],[536,24.7199],[537,24.5824],[538,24.515],[539,24.6557],[540,25.101],[541,25.8267],[542,26.8042],[543,27.993],[544,29.3657],[545,30.8716],[549,37.2221],[550,38.6955],[551,40.0357],[552,41.1884],[555,44.161],[558,46.8051],[561,49.1106],[563,50.3806],[565,51.3555],[567,51.978],[570,52.4162],[572,52.4496],[578,51.9955],[595,53.6207],[615,58.9183],[620,59.6551],[626,59.7424],[632,59.0183],[634,58.5922],[676,57.2288],[677,56.0251],[678,54.6238],[680,52.8675],[685,49.5273],[688,47.8502],[690,46.9594],[694,45.6571],[696,45.1811],[699,44.855],[703,44.8621],[711,45.8862],[727,48.7987],[733,49.2351],[740,48.8878],[751,47.0836],[755,46.0974],[757,45.4487],[759,44.5729],[770,57.6819],[777,59.2908],[783,59.8053],[789,59.4816],[797,57.8533],[803,56.1174],[834,57.1982],[887,58.3573],[895,56.1174],[926,57.1982],[929,57.2288]]
       }
       foot_r_ik_ctrl.Rotation.Y {
        type=MovieSceneScriptingFloatChannel
        default=0
        thinned_from=753
        keys [[-2,3.268],[-1,3.7614],[0,4.0477],[1,4.245],[2,4.4289],[3,4.5921],[4,4.7404],[6,5.0027],[8,5.2284],[10,5.4116],[11,5.4903],[12,5.5562],[13,5.6071],[14,5.6328],[15,5.6379],[16,5.6292],[17,5.5972],[18,5.5471],[19,5.4852],[20,5.4109],[21,5.3224],[22,5.211],[24,4.9351],[25,4.7793],[27,4.4269],[32,3.4832],[34,3.1328],[35,2.9646],[37,2.6545],[38,2.5103],[41,2.109],[42,1.9945],[43,1.8741],[45,1.6563],[46,1.5445],[48,1.3374],[49,1.2291],[51,1.0187],[52,0.9007],[53,0.7866],[54,0.6628],[55,0.5337],[56,0.4098],[57,0.279],[58,0.1549],[59,0.0335],[60,-0.0904],[61,-0.2019],[62,-0.3126],[63,-0.4197],[64,-0.5185],[65,-0.6258],[66,-0.7348],[67,-0.8476],[69,-1.0698],[70,-1.1768],[71,-1.2788],[72,-1.3778],[73,-1.4675],[74,-1.5483],[75,-1.6199],[76,-1.6781],[77,-1.7274],[78,-1.7626],[79,-1.7802],[80,-1.7903],[81,-1.8164],[82,-1.8723],[83,-1.9548],[84,-2.0525],[85,-2.1577],[86,-2.2152],[87,-2.1827],[88,-2.0889],[89,-1.9526],[90,-1.7996],[91,-1.6537],[92,-1.5306],[93,-1.4441],[94,-1.4021],[95,-1.4075],[96,-1.4747],[97,-1.6033],[98,-1.7778],[99,-1.9758],[100,-2.1805],[101,-2.3686],[102,-2.5419],[103,-2.7014],[104,-2.8475],[105,-2.9827],[108,-3.3672],[109,-3.5002],[110,-3.6403],[111,-3.7932],[112,-3.9595],[113,-4.1432],[119,-5.3082],[121,-5.6609],[123,-5.9764],[124,-6.1192],[125,-6.2496],[127,-6.4672],[128,-6.5503],[129,-6.6199],[130,-6.6728],[131,-6.7107],[132,-6.7339],[134,-6.7371],[135,-6.7135],[138,-6.574],[142,-6.3146],[150,-5.7062],[153,-5.5096],[155,-5.4039],[159,-5.2705],[168,-5.0867],[174,-5.0266],[176,-5.0265],[179,-5.056],[185,-5.2096],[196,-5.545],[201,-5.6395],[208,-5.6897],[211,-5.7414],[213,-5.7962],[214,-5.8384],[215,-5.8975],[219,-6.188],[221,-6.3159],[229,-6.9804],[233,-7.3697],[235,-7.594],[237,-7.8389],[239,-8.0232],[242,-8.2137],[243,-8.2505],[245,-8.2641],[247,-8.2149],[249,-8.1323],[251,-7.9977],[252,-7.9102],[253,-7.8013],[254,-7.671],[255,-7.5149],[256,-7.3343],[257,-7.1219],[258,-6.8635],[259,-6.5646],[260,-6.2255],[261,-5.8534],[262,-5.4459],[263,-5.0062],[264,-4.5388],[265,-4.0477],[266,-3.5469],[267,-3.0352],[269,-1.9928],[270,-1.4662],[272,-0.4072],[273,0.1201],[274,0.6413],[275,1.1523],[276,1.6414],[277,2.1198],[278,2.5738],[279,3.0029],[280,3.4009],[281,3.7707],[282,4.106],[283,4.4003],[284,4.657],[285,4.8635],[286,5.0284],[287,5.145],[288,5.2275],[290,5.3532],[291,5.3977],[292,5.4297],[293,5.4501],[294,5.4566],[295,5.4476],[296,5.4268],[297,5.3936],[299,5.2882],[301,5.1396],[303,4.9572],[305,4.7414],[307,4.4939],[309,4.2212],[323,3.268],[324,3.7614],[325,4.0477],[326,4.245],[327,4.4289],[328,4.5921],[329,4.7404],[331,5.0027],[333,5.2284],[335,5.4116],[336,5.4903],[337,5.5562],[338,5.6071],[339,5.6328],[340,5.6379],[341,5.6292],[342,5.5972],[343,5.5471],[344,5.4852],[345,5.4109],[346,5.3224],[347,5.211],[349,4.9351],[350,4.7793],[352,4.4269],[357,3.4832],[359,3.1328],[360,2.9646],[362,2.6545],[363,2.5103],[366,2.109],[367,1.9945],[368,1.8741],[370,1.6563],[371,1.5445],[373,1.3374],[374,1.2291],[376,1.0187],[377,0.9007],[378,0.7866],[379,0.6628],[380,0.5337],[381,0.4098],[382,0.279],[383,0.1549],[384,0.0335],[385,-0.0904],[386,-0.2019],[387,-0.3126],[388,-0.4197],[389,-0.5185],[390,-0.6258],[391,-0.7348],[392,-0.8476],[394,-1.0698],[395,-1.1768],[396,-1.2788],[397,-1.3778],[398,-1.4675],[399,-1.5483],[400,-1.6199],[401,-1.6781],[402,-1.7274],[403,-1.7626],[404,-1.7802],[405,-1.7903],[406,-1.8164],[407,-1.8723],[408,-1.9548],[409,-2.0525],[410,-2.1577],[411,-2.2152],[412,-2.1827],[413,-2.0889],[414,-1.9526],[415,-1.7996],[416,-1.6537],[417,-1.5306],[418,-1.4441],[419,-1.4021],[420,-1.4075],[421,-1.4747],[422,-1.6033],[423,-1.7778],[424,-1.9758],[425,-2.1805],[426,-2.3686],[427,-2.5419],[428,-2.7014],[429,-2.8475],[430,-2.9827],[433,-3.3672],[434,-3.5002],[435,-3.6403],[436,-3.7932],[437,-3.9595],[438,-4.1432],[444,-5.3082],[446,-5.6609],[448,-5.9764],[449,-6.1192],[450,-6.2496],[452,-6.4672],[453,-6.5503],[454,-6.6199],[455,-6.6728],[456,-6.7107],[457,-6.7339],[459,-6.7371],[460,-6.7135],[463,-6.574],[467,-6.3146],[475,-5.7062],[478,-5.5096],[480,-5.4039],[484,-5.2705],[493,-5.0867],[499,-5.0266],[501,-5.0265],[504,-5.056],[510,-5.2096],[521,-5.545],[526,-5.6395],[533,-5.6897],[536,-5.7414],[538,-5.7962],[539,-5.8384],[540,-5.8975],[544,-6.188],[546,-6.3159],[554,-6.9804],[558,-7.3697],[560,-7.594],[562,-7.8389],[564,-8.0232],[567,-8.2137],[568,-8.2505],[570,-8.2641],[572,-8.2149],[574,-8.1323],[576,-7.9977],[577,-7.9102],[578,-7.8013],[579,-7.671],[580,-7.5149],[581,-7.3343],[582,-7.1219],[583,-6.8635],[584,-6.5646],[585,-6.2255],[586,-5.8534],[587,-5.4459],[588,-5.0062],[589,-4.5388],[590,-4.0477],[591,-3.5469],[592,-3.0352],[594,-1.9928],[595,-1.4662],[597,-0.4072],[598,0.1201],[599,0.6413],[600,1.1523],[601,1.6414],[602,2.1198],[603,2.5738],[604,3.0029],[605,3.4009],[606,3.7707],[607,4.106],[608,4.4003],[609,4.657],[610,4.8635],[611,5.0284],[612,5.145],[613,5.2275],[615,5.3532],[616,5.3977],[617,5.4297],[618,5.4501],[619,5.4566],[620,5.4476],[621,5.4268],[622,5.3936],[624,5.2882],[626,5.1396],[628,4.9572],[630,4.7414],[632,4.4939],[634,4.2212],[676,3.268],[677,3.7614],[678,4.0477],[679,4.245],[680,4.4289],[681,4.5921],[682,4.7404],[684,5.0027],[686,5.2284],[688,5.4116],[689,5.4903],[690,5.5562],[691,5.6071],[692,5.6328],[693,5.6379],[694,5.6292],[695,5.5972],[696,5.5471],[697,5.4852],[698,5.4109],[699,5.3224],[700,5.211],[702,4.9351],[703,4.7793],[705,4.4269],[710,3.4832],[712,3.1328],[713,2.9646],[715,2.6545],[716,2.5103],[719,2.109],[720,1.9945],[721,1.8741],[723,1.6563],[724,1.5445],[726,1.3374],[727,1.2291],[729,1.0187],[730,0.9007],[731,0.7866],[732,0.6628],[733,0.5337],[734,0.4098],[735,0.279],[736,0.1549],[737,0.0335],[738,-0.0904],[739,-0.2019],[740,-0.3126],[741,-0.4197],[742,-0.5185],[743,-0.6258],[744,-0.7348],[745,-0.8476],[747,-1.0698],[748,-1.1768],[749,-1.2788],[750,-1.3778],[751,-1.4675],[752,-1.5483],[753,-1.6199],[754,-1.6781],[755,-1.7274],[757,-1.7802],[759,-1.8164],[770,4.8635],[771,5.0284],[772,5.145],[773,5.2275],[775,5.3532],[776,5.3977],[777,5.4297],[778,5.4501],[779,5.4566],[780,5.4476],[781,5.4268],[782,5.3936],[784,5.2882],[786,5.1396],[788,4.9572],[790,4.7414],[792,4.4939],[794,4.2212],[803,2.894],[834,3.2556],[837,3.268],[887,4.0779],[895,2.894],[926,3.2556],[929,3.268]]
       }
       foot_r_ik_ctrl.Rotation.Z {
        type=MovieSceneScriptingFloatChannel
        default=0
        thinned_from=753
        keys [[-2,8.686],[-1,9.3606],[0,9.5094],[2,9.9334],[5,10.4796],[7,10.8009],[10,11.2064],[13,11.5039],[14,11.6209],[16,11.9278],[22,12.9947],[24,13.2953],[26,13.5227],[27,13.6011],[28,13.6483],[29,13.6665],[30,13.6492],[31,13.5924],[32,13.5007],[33,13.3592],[34,13.1866],[35,12.983],[37,12.4781],[39,11.8781],[41,11.2201],[47,9.1146],[49,8.4697],[50,8.1693],[51,7.8886],[52,7.6382],[53,7.4073],[54,7.2129],[55,7.0404],[56,6.8835],[57,6.7487],[58,6.6295],[59,6.5241],[61,6.3698],[64,6.2467],[66,6.218],[69,6.2312],[72,6.3121],[75,6.4305],[76,6.4865],[77,6.5639],[78,6.6574],[79,6.7741],[80,6.9103],[81,7.0731],[82,7.2737],[83,7.508],[84,7.7832],[86,8.4533],[87,8.7428],[88,8.9839],[89,9.1642],[90,9.2869],[91,9.353],[92,9.3757],[95,9.3251],[96,9.3369],[97,9.3927],[98,9.5108],[99,9.7099],[102,10.4228],[104,10.8488],[106,11.2173],[120,13.3676],[124,14.0892],[129,15.0607],[131,15.3744],[133,15.6077],[156,16.9371],[164,17.6556],[171,18.5477],[189,21.3877],[203,22.9272],[211,23.4873],[214,23.5593],[216,23.4849],[217,23.3715],[218,23.1668],[219,22.8736],[220,22.4892],[221,22.0239],[222,21.4895],[223,20.8805],[224,20.2064],[225,19.4636],[226,18.6612],[227,17.8023],[228,16.8874],[229,16.2529],[231,14.8809],[233,13.3823],[234,12.5924],[235,11.7563],[236,10.8414],[237,9.8499],[238,8.7925],[239,7.6852],[240,6.5294],[241,5.3393],[242,4.1261],[243,2.8962],[244,1.6545],[245,0.4168],[246,-0.8139],[247,-2.0195],[248,-3.2023],[249,-4.3427],[250,-5.444],[251,-6.4951],[252,-7.4911],[253,-8.4551],[254,-9.388],[255,-10.2903],[256,-11.1572],[257,-11.9846],[258,-12.7697],[259,-13.5148],[260,-14.2053],[261,-14.8431],[262,-15.4187],[263,-15.9292],[264,-16.3721],[268,-17.9243],[272,-19.3129],[275,-20.227],[278,-20.9973],[281,-21.5896],[284,-21.9933],[287,-22.2084],[290,-22.2448],[292,-22.1563],[294,-21.9521],[296,-21.649],[299,-21.0405],[304,-19.749],[309,-18.3083],[323,8.686],[324,9.3606],[325,9.5094],[327,9.9334],[330,10.4796],[332,10.8009],[335,11.2064],[338,11.5039],[339,11.6209],[341,11.9278],[347,12.9947],[349,13.2953],[351,13.5227],[352,13.6011],[353,13.6483],[354,13.6665],[355,13.6492],[356,13.5924],[357,13.5007],[358,13.3592],[359,13.1866],[360,12.983],[362,12.4781],[364,11.8781],[366,11.2201],[372,9.1146],[374,8.4697],[375,8.1693],[376,7.8886],[377,7.6382],[378,7.4073],[379,7.2129],[380,7.0404],[381,6.8835],[382,6.7487],[383,6.6295],[384,6.5241],[386,6.3698],[389,6.2467],[391,6.218],[394,6.2312],[397,6.3121],[400,6.4305],[401,6.4865],[402,6.5639],[403,6.6574],[404,6.7741],[405,6.9103],[406,7.0731],[407,7.2737],[408,7.508],[409,7.7832],[411,8.4533],[412,8.7428],[413,8.9839],[414,9.1642],[415,9.2869],[416,9.353],[417,9.3757],[420,9.3251],[421,9.3369],[422,9.3927],[423,9.5108],[424,9.7099],[427,10.4228],[429,10.8488],[431,11.2173],[445,13.3676],[449,14.0892],[454,15.0607],[456,15.3744],[458,15.6077],[481,16.9371],[489,17.6556],[496,18.5477],[514,21.3877],[528,22.9272],[536,23.4873],[539,23.5593],[541,23.4849],[542,23.3715],[543,23.1668],[544,22.8736],[545,22.4892],[546,22.0239],[547,21.4895],[548,20.8805],[549,20.2064],[550,19.4636],[551,18.6612],[552,17.8023],[553,16.8874],[554,16.2529],[556,14.8809],[558,13.3823],[559,12.5924],[560,11.7563],[561,10.8414],[562,9.8499],[563,8.7925],[564,7.6852],[565,6.5294],[566,5.3393],[567,4.1261],[568,2.8962],[569,1.6545],[570,0.4168],[571,-0.8139],[572,-2.0195],[573,-3.2023],[574,-4.3427],[575,-5.444],[576,-6.4951],[577,-7.4911],[578,-8.4551],[579,-9.388],[580,-10.2903],[581,-11.1572],[582,-11.9846],[583,-12.7697],[584,-13.5148],[585,-14.2053],[586,-14.8431],[587,-15.4187],[588,-15.9292],[589,-16.3721],[593,-17.9243],[597,-19.3129],[600,-20.227],[603,-20.9973],[606,-21.5896],[609,-21.9933],[612,-22.2084],[615,-22.2448],[617,-22.1563],[619,-21.9521],[621,-21.649],[624,-21.0405],[629,-19.749],[634,-18.3083],[676,8.686],[677,9.3606],[678,9.5094],[680,9.9334],[683,10.4796],[685,10.8009],[688,11.2064],[691,11.5039],[692,11.6209],[694,11.9278],[700,12.9947],[702,13.2953],[704,13.5227],[705,13.6011],[706,13.6483],[707,13.6665],[708,13.6492],[709,13.5924],[710,13.5007],[711,13.3592],[712,13.1866],[713,12.983],[715,12.4781],[717,11.8781],[719,11.2201],[725,9.1146],[727,8.4697],[728,8.1693],[729,7.8886],[730,7.6382],[731,7.4073],[732,7.2129],[733,7.0404],[734,6.8835],[735,6.7487],[736,6.6295],[737,6.5241],[739,6.3698],[742,6.2467],[744,6.218],[747,6.2312],[750,6.3121],[753,6.4305],[754,6.4865],[755,6.5639],[757,6.7741],[759,7.0731],[770,-22.0889],[773,-22.2384],[776,-22.216],[778,-22.0669],[780,-21.8117],[783,-21.2619],[787,-20.2946],[803,-15.6489],[834,8.2773],[837,8.686],[887,-18.0242],[895,-15.6489],[926,8.2773],[929,8.686]]
       }
       foot_r_ik_ctrl.Scale.X {type=MovieSceneScriptingFloatChannel, default=1, const [1,753]}
       foot_r_ik_ctrl.Scale.Y {type=MovieSceneScriptingFloatChannel, default=1, const [1,753]}
       foot_r_ik_ctrl.Scale.Z {type=MovieSceneScriptingFloatChannel, default=1, const [1,753]}
       ball_r_ik_ctrl.Location.X {type=MovieSceneScriptingFloatChannel, default=0, thinned_from=751, keys [[-2,0],[929,0]]}
       ball_r_ik_ctrl.Location.Y {type=MovieSceneScriptingFloatChannel, default=0, thinned_from=751, keys [[-2,0],[929,0]]}
       ball_r_ik_ctrl.Location.Z {type=MovieSceneScriptingFloatChannel, default=0, thinned_from=751, keys [[-2,0],[929,0]]}
       ball_r_ik_ctrl.Rotation.X {
        type=MovieSceneScriptingFloatChannel
        default=0
        thinned_from=751
        keys [[-2,0.549],[-1,0.5323],[0,0.5323],[1,0.5155],[2,0.5155],[3,0.4985],[4,0.4985],[5,0.4819],[7,0.4819],[8,0.4651],[10,0.4651],[11,0.4486],[13,0.4486],[14,0.4319],[17,0.4327],[18,0.421],[20,0.421],[21,0.4142],[26,0.4148],[27,0.421],[30,0.421],[31,0.4187],[32,0.4187],[33,0.4338],[38,0.4338],[39,0.4424],[42,0.4424],[43,0.4539],[45,0.4539],[46,0.4595],[47,0.4585],[49,0.4585],[50,0.4635],[54,0.4635],[55,0.4684],[62,0.4684],[63,0.47],[64,0.4644],[72,0.4644],[73,0.4539],[78,0.4539],[79,0.4525],[81,0.4525],[82,0.4423],[84,0.4423],[85,0.4217],[86,0.4217],[87,0.3961],[90,0.3351],[91,0.3198],[92,0.3198],[93,0.2998],[94,0.2998],[95,0.2895],[97,0.2895],[98,0.2805],[115,0.2807],[119,0.2803],[120,0.2691],[121,0.2691],[122,0.2479],[123,0.2479],[124,0.2267],[125,0.2267],[126,0.2053],[127,0.2095],[128,0.1858],[129,0.1858],[130,0.1701],[131,0.1701],[132,0.1543],[158,0.1545],[159,0.1517],[169,0.1517],[170,0.1382],[174,0.1382],[175,0.1423],[176,0.1207],[184,0.1207],[185,0.1099],[190,0.1099],[191,0.1113],[206,0.1113],[207,0.1151],[208,0.0957],[215,0.0957],[222,0.2311],[223,0.2513],[226,0.3031],[227,0.3117],[228,0.3117],[229,0.329],[230,0.329],[231,0.3463],[232,0.3463],[233,0.3636],[235,0.3636],[236,0.3806],[238,0.3806],[239,0.3882],[240,0.3882],[241,0.3918],[242,0.3918],[243,0.3926],[246,0.392],[247,0.3903],[248,0.3903],[250,0.3865],[251,0.3865],[252,0.3839],[253,0.3839],[254,0.3829],[255,0.3828],[256,0.3814],[257,0.3814],[258,0.3837],[260,0.3837],[261,0.3858],[266,0.3858],[267,0.3993],[270,0.3993],[271,0.4093],[274,0.4093],[275,0.4269],[277,0.4269],[278,0.4445],[280,0.4445],[281,0.4621],[283,0.4621],[284,0.4797],[285,0.4797],[286,0.4914],[287,0.4931],[288,0.5066],[290,0.5066],[291,0.5201],[293,0.5201],[294,0.5337],[296,0.5337],[297,0.5427],[301,0.5427],[302,0.5563],[303,0.5581],[309,0.5581],[323,0.549],[324,0.5323],[325,0.5323],[326,0.5155],[327,0.5155],[328,0.4985],[329,0.4985],[330,0.4819],[332,0.4819],[333,0.4651],[335,0.4651],[336,0.4486],[338,0.4486],[339,0.4319],[342,0.4327],[343,0.421],[345,0.421],[346,0.4142],[351,0.4148],[352,0.421],[355,0.421],[356,0.4187],[357,0.4187],[358,0.4338],[363,0.4338],[364,0.4424],[367,0.4424],[368,0.4539],[370,0.4539],[371,0.4595],[372,0.4585],[374,0.4585],[375,0.4635],[379,0.4635],[380,0.4684],[387,0.4684],[388,0.47],[389,0.4644],[397,0.4644],[398,0.4539],[403,0.4539],[404,0.4525],[406,0.4525],[407,0.4423],[409,0.4423],[410,0.4217],[411,0.4217],[412,0.3961],[415,0.3351],[416,0.3198],[417,0.3198],[418,0.2998],[419,0.2998],[420,0.2895],[422,0.2895],[423,0.2805],[440,0.2807],[444,0.2803],[445,0.2691],[446,0.2691],[447,0.2479],[448,0.2479],[449,0.2267],[450,0.2267],[451,0.2053],[452,0.2095],[453,0.1858],[454,0.1858],[455,0.1701],[456,0.1701],[457,0.1543],[483,0.1545],[484,0.1517],[494,0.1517],[495,0.1382],[499,0.1382],[500,0.1423],[501,0.1207],[509,0.1207],[510,0.1099],[515,0.1099],[516,0.1113],[531,0.1113],[532,0.1151],[533,0.0957],[540,0.0957],[547,0.2311],[548,0.2513],[551,0.3031],[552,0.3117],[553,0.3117],[554,0.329],[555,0.329],[556,0.3463],[557,0.3463],[558,0.3636],[560,0.3636],[561,0.3806],[563,0.3806],[564,0.3882],[565,0.3882],[566,0.3918],[567,0.3918],[568,0.3926],[571,0.392],[572,0.3903],[573,0.3903],[575,0.3865],[576,0.3865],[577,0.3839],[578,0.3839],[579,0.3829],[580,0.3828],[581,0.3814],[582,0.3814],[583,0.3837],[585,0.3837],[586,0.3858],[591,0.3858],[592,0.3993],[595,0.3993],[596,0.4093],[599,0.4093],[600,0.4269],[602,0.4269],[603,0.4445],[605,0.4445],[606,0.4621],[608,0.4621],[609,0.4797],[610,0.4797],[611,0.4914],[612,0.4931],[613,0.5066],[615,0.5066],[616,0.5201],[618,0.5201],[619,0.5337],[621,0.5337],[622,0.5427],[626,0.5427],[627,0.5563],[628,0.5581],[634,0.5581],[676,0.549],[677,0.5323],[678,0.5323],[679,0.5155],[680,0.5155],[681,0.4985],[682,0.4985],[683,0.4819],[685,0.4819],[686,0.4651],[688,0.4651],[689,0.4486],[691,0.4486],[692,0.4319],[695,0.4327],[696,0.421],[698,0.421],[699,0.4142],[704,0.4148],[705,0.421],[708,0.421],[709,0.4187],[710,0.4187],[711,0.4338],[716,0.4338],[717,0.4424],[720,0.4424],[721,0.4539],[723,0.4539],[724,0.4595],[725,0.4585],[727,0.4585],[728,0.4635],[732,0.4635],[733,0.4684],[740,0.4684],[741,0.47],[742,0.4644],[750,0.4644],[751,0.4539],[755,0.4539],[757,0.4525],[759,0.4525],[770,0.4797],[771,0.4914],[772,0.4931],[773,0.5066],[775,0.5066],[776,0.5201],[778,0.5201],[779,0.5337],[781,0.5337],[782,0.5427],[786,0.5427],[787,0.5563],[788,0.5581],[802,0.5581],[803,0.5423],[894,0.5581],[895,0.5423],[929,0.549]]
       }
       ball_r_ik_ctrl.Rotation.Y {
        type=MovieSceneScriptingFloatChannel
        default=0
        thinned_from=751
        keys [[-2,-0.1434],[-1,-0.1406],[0,-0.1406],[1,-0.1391],[4,-0.1388],[5,-0.1345],[7,-0.1345],[8,-0.1326],[10,-0.1326],[11,-0.1266],[13,-0.1266],[14,-0.1231],[15,-0.1244],[17,-0.1244],[18,-0.1109],[20,-0.1109],[21,-0.1139],[23,-0.1139],[24,-0.1066],[26,-0.1066],[27,-0.1078],[32,-0.1078],[33,-0.1063],[38,-0.1063],[39,-0.1118],[42,-0.1118],[43,-0.1176],[45,-0.1176],[46,-0.1228],[47,-0.1225],[49,-0.1225],[50,-0.1258],[54,-0.1258],[55,-0.1307],[62,-0.1307],[63,-0.1326],[64,-0.1361],[72,-0.1361],[73,-0.1315],[78,-0.1315],[79,-0.1337],[81,-0.1337],[82,-0.1313],[84,-0.1313],[85,-0.129],[86,-0.129],[87,-0.1236],[88,-0.1194],[89,-0.1136],[90,-0.1082],[91,-0.1047],[92,-0.1047],[93,-0.097],[94,-0.097],[95,-0.091],[97,-0.091],[98,-0.0868],[110,-0.0868],[111,-0.0926],[115,-0.0926],[116,-0.0957],[119,-0.0957],[120,-0.1008],[121,-0.1008],[122,-0.1005],[123,-0.1005],[124,-0.1002],[125,-0.1002],[126,-0.1009],[128,-0.1002],[129,-0.1002],[130,-0.0994],[142,-0.0995],[143,-0.0987],[158,-0.0987],[169,-0.0986],[170,-0.0975],[175,-0.0975],[176,-0.0955],[184,-0.0955],[185,-0.0943],[206,-0.0942],[207,-0.0945],[208,-0.0935],[215,-0.0935],[217,-0.0977],[220,-0.1044],[221,-0.107],[222,-0.1088],[223,-0.1111],[224,-0.1131],[225,-0.1147],[226,-0.1167],[227,-0.1178],[228,-0.1178],[229,-0.1192],[230,-0.1192],[231,-0.1211],[232,-0.1211],[233,-0.1231],[235,-0.1231],[236,-0.1273],[238,-0.1273],[239,-0.1397],[240,-0.1397],[241,-0.1534],[242,-0.1534],[243,-0.1713],[244,-0.1713],[245,-0.189],[246,-0.189],[247,-0.2098],[248,-0.2098],[249,-0.2243],[250,-0.2368],[251,-0.2368],[252,-0.2589],[253,-0.2589],[254,-0.2766],[255,-0.2747],[256,-0.2918],[257,-0.2918],[258,-0.3072],[260,-0.3072],[261,-0.3246],[266,-0.3246],[267,-0.3319],[270,-0.3319],[271,-0.3277],[290,-0.3291],[301,-0.3296],[302,-0.3289],[309,-0.3293],[323,-0.1434],[324,-0.1406],[325,-0.1406],[326,-0.1391],[329,-0.1388],[330,-0.1345],[332,-0.1345],[333,-0.1326],[335,-0.1326],[336,-0.1266],[338,-0.1266],[339,-0.1231],[340,-0.1244],[342,-0.1244],[343,-0.1109],[345,-0.1109],[346,-0.1139],[348,-0.1139],[349,-0.1066],[351,-0.1066],[352,-0.1078],[357,-0.1078],[358,-0.1063],[363,-0.1063],[364,-0.1118],[367,-0.1118],[368,-0.1176],[370,-0.1176],[371,-0.1228],[372,-0.1225],[374,-0.1225],[375,-0.1258],[379,-0.1258],[380,-0.1307],[387,-0.1307],[388,-0.1326],[389,-0.1361],[397,-0.1361],[398,-0.1315],[403,-0.1315],[404,-0.1337],[406,-0.1337],[407,-0.1313],[409,-0.1313],[410,-0.129],[411,-0.129],[412,-0.1236],[413,-0.1194],[414,-0.1136],[415,-0.1082],[416,-0.1047],[417,-0.1047],[418,-0.097],[419,-0.097],[420,-0.091],[422,-0.091],[423,-0.0868],[435,-0.0868],[436,-0.0926],[440,-0.0926],[441,-0.0957],[444,-0.0957],[445,-0.1008],[446,-0.1008],[447,-0.1005],[448,-0.1005],[449,-0.1002],[450,-0.1002],[451,-0.1009],[453,-0.1002],[454,-0.1002],[455,-0.0994],[467,-0.0995],[468,-0.0987],[483,-0.0987],[494,-0.0986],[495,-0.0975],[500,-0.0975],[501,-0.0955],[509,-0.0955],[510,-0.0943],[531,-0.0942],[532,-0.0945],[533,-0.0935],[540,-0.0935],[542,-0.0977],[545,-0.1044],[546,-0.107],[547,-0.1088],[548,-0.1111],[549,-0.1131],[550,-0.1147],[551,-0.1167],[552,-0.1178],[553,-0.1178],[554,-0.1192],[555,-0.1192],[556,-0.1211],[557,-0.1211],[558,-0.1231],[560,-0.1231],[561,-0.1273],[563,-0.1273],[564,-0.1397],[565,-0.1397],[566,-0.1534],[567,-0.1534],[568,-0.1713],[569,-0.1713],[570,-0.189],[571,-0.189],[572,-0.2098],[573,-0.2098],[574,-0.2243],[575,-0.2368],[576,-0.2368],[577,-0.2589],[578,-0.2589],[579,-0.2766],[580,-0.2747],[581,-0.2918],[582,-0.2918],[583,-0.3072],[585,-0.3072],[586,-0.3246],[591,-0.3246],[592,-0.3319],[595,-0.3319],[596,-0.3277],[615,-0.3291],[626,-0.3296],[627,-0.3289],[634,-0.3293],[676,-0.1434],[677,-0.1406],[678,-0.1406],[679,-0.1391],[682,-0.1388],[683,-0.1345],[685,-0.1345],[686,-0.1326],[688,-0.1326],[689,-0.1266],[691,-0.1266],[692,-0.1231],[693,-0.1244],[695,-0.1244],[696,-0.1109],[698,-0.1109],[699,-0.1139],[701,-0.1139],[702,-0.1066],[704,-0.1066],[705,-0.1078],[710,-0.1078],[711,-0.1063],[716,-0.1063],[717,-0.1118],[720,-0.1118],[721,-0.1176],[723,-0.1176],[724,-0.1228],[725,-0.1225],[727,-0.1225],[728,-0.1258],[732,-0.1258],[733,-0.1307],[740,-0.1307],[741,-0.1326],[742,-0.1361],[750,-0.1361],[751,-0.1315],[755,-0.1315],[757,-0.1337],[759,-0.1337],[770,-0.3286],[786,-0.3296],[787,-0.3289],[802,-0.3293],[803,-0.3288],[837,-0.1434],[887,-0.3293],[894,-0.3293],[895,-0.3288],[929,-0.1434]]
       }
       ball_r_ik_ctrl.Rotation.Z {
        type=MovieSceneScriptingFloatChannel
        default=0
        thinned_from=751
        keys [[-2,-0.1031],[-1,-0.1044],[2,-0.1044],[3,-0.1033],[4,-0.1033],[5,-0.1003],[7,-0.1003],[8,-0.095],[10,-0.095],[11,-0.0879],[13,-0.0879],[14,-0.0783],[15,-0.0808],[17,-0.0808],[18,-0.0623],[20,-0.0623],[21,-0.0486],[23,-0.0486],[24,-0.0352],[26,-0.0352],[27,-0.0243],[30,-0.0243],[31,-0.0249],[32,-0.0249],[33,-0.0159],[38,-0.0158],[39,-0.0308],[42,-0.0308],[43,-0.0458],[45,-0.0458],[46,-0.0607],[47,-0.0592],[49,-0.0592],[50,-0.0745],[54,-0.0745],[55,-0.0882],[62,-0.0882],[63,-0.0888],[64,-0.1069],[72,-0.1069],[73,-0.1032],[78,-0.1032],[79,-0.1043],[81,-0.1043],[82,-0.1049],[84,-0.1049],[85,-0.1045],[86,-0.1045],[87,-0.1077],[88,-0.1117],[89,-0.1181],[90,-0.125],[91,-0.1312],[92,-0.1312],[93,-0.1461],[94,-0.1461],[95,-0.16],[97,-0.16],[98,-0.1731],[110,-0.1731],[111,-0.1606],[115,-0.1606],[116,-0.1487],[119,-0.1487],[120,-0.1362],[121,-0.1362],[122,-0.1304],[123,-0.1304],[124,-0.1245],[125,-0.1245],[126,-0.1198],[127,-0.1202],[128,-0.1148],[129,-0.1148],[130,-0.1106],[131,-0.1106],[132,-0.1073],[142,-0.1073],[143,-0.1077],[158,-0.1077],[159,-0.108],[169,-0.108],[170,-0.106],[174,-0.106],[175,-0.1058],[184,-0.1059],[185,-0.105],[190,-0.105],[206,-0.1051],[207,-0.1054],[208,-0.1048],[215,-0.1048],[217,-0.1039],[218,-0.1044],[219,-0.104],[220,-0.1044],[222,-0.1045],[224,-0.1054],[225,-0.1053],[227,-0.1062],[230,-0.1063],[231,-0.1067],[232,-0.1067],[233,-0.1071],[235,-0.1071],[236,-0.1087],[238,-0.1087],[239,-0.1172],[240,-0.1172],[241,-0.127],[242,-0.127],[243,-0.14],[244,-0.14],[245,-0.153],[246,-0.153],[247,-0.1706],[248,-0.1706],[249,-0.1791],[250,-0.1895],[251,-0.1895],[252,-0.2058],[253,-0.2058],[254,-0.2188],[255,-0.2169],[256,-0.2285],[257,-0.2285],[258,-0.242],[260,-0.242],[261,-0.2536],[266,-0.2536],[267,-0.2618],[270,-0.2618],[271,-0.2609],[274,-0.2609],[275,-0.263],[277,-0.263],[278,-0.2651],[280,-0.2651],[281,-0.267],[283,-0.267],[284,-0.2689],[285,-0.2689],[286,-0.2705],[287,-0.2702],[288,-0.2718],[290,-0.2718],[291,-0.2728],[293,-0.2728],[294,-0.2744],[296,-0.2744],[297,-0.2755],[301,-0.2755],[302,-0.2773],[309,-0.2771],[323,-0.1031],[324,-0.1044],[327,-0.1044],[328,-0.1033],[329,-0.1033],[330,-0.1003],[332,-0.1003],[333,-0.095],[335,-0.095],[336,-0.0879],[338,-0.0879],[339,-0.0783],[340,-0.0808],[342,-0.0808],[343,-0.0623],[345,-0.0623],[346,-0.0486],[348,-0.0486],[349,-0.0352],[351,-0.0352],[352,-0.0243],[355,-0.0243],[356,-0.0249],[357,-0.0249],[358,-0.0159],[363,-0.0158],[364,-0.0308],[367,-0.0308],[368,-0.0458],[370,-0.0458],[371,-0.0607],[372,-0.0592],[374,-0.0592],[375,-0.0745],[379,-0.0745],[380,-0.0882],[387,-0.0882],[388,-0.0888],[389,-0.1069],[397,-0.1069],[398,-0.1032],[403,-0.1032],[404,-0.1043],[406,-0.1043],[407,-0.1049],[409,-0.1049],[410,-0.1045],[411,-0.1045],[412,-0.1077],[413,-0.1117],[414,-0.1181],[415,-0.125],[416,-0.1312],[417,-0.1312],[418,-0.1461],[419,-0.1461],[420,-0.16],[422,-0.16],[423,-0.1731],[435,-0.1731],[436,-0.1606],[440,-0.1606],[441,-0.1487],[444,-0.1487],[445,-0.1362],[446,-0.1362],[447,-0.1304],[448,-0.1304],[449,-0.1245],[450,-0.1245],[451,-0.1198],[452,-0.1202],[453,-0.1148],[454,-0.1148],[455,-0.1106],[456,-0.1106],[457,-0.1073],[467,-0.1073],[468,-0.1077],[483,-0.1077],[484,-0.108],[494,-0.108],[495,-0.106],[499,-0.106],[500,-0.1058],[509,-0.1059],[510,-0.105],[515,-0.105],[531,-0.1051],[532,-0.1054],[533,-0.1048],[540,-0.1048],[542,-0.1039],[543,-0.1044],[544,-0.104],[545,-0.1044],[547,-0.1045],[549,-0.1054],[550,-0.1053],[552,-0.1062],[555,-0.1063],[556,-0.1067],[557,-0.1067],[558,-0.1071],[560,-0.1071],[561,-0.1087],[563,-0.1087],[564,-0.1172],[565,-0.1172],[566,-0.127],[567,-0.127],[568,-0.14],[569,-0.14],[570,-0.153],[571,-0.153],[572,-0.1706],[573,-0.1706],[574,-0.1791],[575,-0.1895],[576,-0.1895],[577,-0.2058],[578,-0.2058],[579,-0.2188],[580,-0.2169],[581,-0.2285],[582,-0.2285],[583,-0.242],[585,-0.242],[586,-0.2536],[591,-0.2536],[592,-0.2618],[595,-0.2618],[596,-0.2609],[599,-0.2609],[600,-0.263],[602,-0.263],[603,-0.2651],[605,-0.2651],[606,-0.267],[608,-0.267],[609,-0.2689],[610,-0.2689],[611,-0.2705],[612,-0.2702],[613,-0.2718],[615,-0.2718],[616,-0.2728],[618,-0.2728],[619,-0.2744],[621,-0.2744],[622,-0.2755],[626,-0.2755],[627,-0.2773],[634,-0.2771],[676,-0.1031],[677,-0.1044],[680,-0.1044],[681,-0.1033],[682,-0.1033],[683,-0.1003],[685,-0.1003],[686,-0.095],[688,-0.095],[689,-0.0879],[691,-0.0879],[692,-0.0783],[693,-0.0808],[695,-0.0808],[696,-0.0623],[698,-0.0623],[699,-0.0486],[701,-0.0486],[702,-0.0352],[704,-0.0352],[705,-0.0243],[708,-0.0243],[709,-0.0249],[710,-0.0249],[711,-0.0159],[716,-0.0158],[717,-0.0308],[720,-0.0308],[721,-0.0458],[723,-0.0458],[724,-0.0607],[725,-0.0592],[727,-0.0592],[728,-0.0745],[732,-0.0745],[733,-0.0882],[740,-0.0882],[741,-0.0888],[742,-0.1069],[750,-0.1069],[751,-0.1032],[755,-0.1032],[757,-0.1043],[759,-0.1043],[770,-0.2689],[771,-0.2705],[772,-0.2702],[773,-0.2718],[775,-0.2718],[776,-0.2728],[778,-0.2728],[779,-0.2744],[781,-0.2744],[782,-0.2755],[786,-0.2755],[787,-0.2773],[802,-0.2771],[803,-0.2766],[837,-0.1031],[887,-0.2771],[894,-0.2771],[895,-0.2766],[929,-0.1031]]
       }
       ball_r_ik_ctrl.Scale.X {type=MovieSceneScriptingFloatChannel, default=1, const [1,751]}
       ball_r_ik_ctrl.Scale.Y {type=MovieSceneScriptingFloatChannel, default=1, const [1,751]}
       ball_r_ik_ctrl.Scale.Z {type=MovieSceneScriptingFloatChannel, default=1, const [1,751]}
       foot_bk1_r_ctrl.Location.X {type=MovieSceneScriptingFloatChannel, default=0, const [0,751]}
       foot_bk1_r_ctrl.Location.Y {type=MovieSceneScriptingFloatChannel, default=0, const [0,751]}
       foot_bk1_r_ctrl.Location.Z {type=MovieSceneScriptingFloatChannel, default=0, const [0,751]}
       foot_bk1_r_ctrl.Rotation.X {type=MovieSceneScriptingFloatChannel, default=0, const [0,751]}
       foot_bk1_r_ctrl.Rotation.Y {type=MovieSceneScriptingFloatChannel, default=0, const [0,751]}
       foot_bk1_r_ctrl.Rotation.Z {type=MovieSceneScriptingFloatChannel, default=0, const [0,751]}
       foot_bk1_r_ctrl.Scale.X {type=MovieSceneScriptingFloatChannel, default=1, const [1,751]}
       foot_bk1_r_ctrl.Scale.Y {type=MovieSceneScriptingFloatChannel, default=1, const [1,751]}
       foot_bk1_r_ctrl.Scale.Z {type=MovieSceneScriptingFloatChannel, default=1, const [1,751]}
       foot_roll_r_ctrl.Location.X {type=MovieSceneScriptingFloatChannel, default=0, const [0,751]}
       foot_roll_r_ctrl.Location.Y {type=MovieSceneScriptingFloatChannel, default=0, const [0,751]}
       foot_roll_r_ctrl.Location.Z {type=MovieSceneScriptingFloatChannel, default=0, const [0,751]}
       foot_roll_r_ctrl.Rotation.X {type=MovieSceneScriptingFloatChannel, default=0, const [0,751]}
       foot_roll_r_ctrl.Rotation.Y {type=MovieSceneScriptingFloatChannel, default=0, const [0,751]}
       foot_roll_r_ctrl.Rotation.Z {type=MovieSceneScriptingFloatChannel, default=0, const [0,751]}
       foot_roll_r_ctrl.Scale.X {type=MovieSceneScriptingFloatChannel, default=1, const [1,751]}
       foot_roll_r_ctrl.Scale.Y {type=MovieSceneScriptingFloatChannel, default=1, const [1,751]}
       foot_roll_r_ctrl.Scale.Z {type=MovieSceneScriptingFloatChannel, default=1, const [1,751]}
       roll_blend_r_ctrl {type=MovieSceneScriptingFloatChannel, default=1, const [1,751]}
       tip_r_ctrl.Location.X {type=MovieSceneScriptingFloatChannel, default=0, const [0,751]}
       tip_r_ctrl.Location.Y {type=MovieSceneScriptingFloatChannel, default=0, const [0,751]}
       tip_r_ctrl.Location.Z {type=MovieSceneScriptingFloatChannel, default=0, const [0,751]}
       tip_r_ctrl.Rotation.X {type=MovieSceneScriptingFloatChannel, default=0, const [0,751]}
       tip_r_ctrl.Rotation.Y {type=MovieSceneScriptingFloatChannel, default=0, const [0,751]}
       tip_r_ctrl.Rotation.Z {type=MovieSceneScriptingFloatChannel, default=0, const [0,751]}
       tip_r_ctrl.Scale.X {type=MovieSceneScriptingFloatChannel, default=1, const [1,751]}
       tip_r_ctrl.Scale.Y {type=MovieSceneScriptingFloatChannel, default=1, const [1,751]}
       tip_r_ctrl.Scale.Z {type=MovieSceneScriptingFloatChannel, default=1, const [1,751]}
       heel_r_ctrl.Location.X {type=MovieSceneScriptingFloatChannel, default=0, const [0,445]}
       heel_r_ctrl.Location.Y {type=MovieSceneScriptingFloatChannel, default=0, const [0,445]}
       heel_r_ctrl.Location.Z {type=MovieSceneScriptingFloatChannel, default=0, const [0,445]}
       heel_r_ctrl.Rotation.X {type=MovieSceneScriptingFloatChannel, default=0, const [0,445]}
       heel_r_ctrl.Rotation.Y {type=MovieSceneScriptingFloatChannel, default=0, const [0,445]}
       heel_r_ctrl.Rotation.Z {type=MovieSceneScriptingFloatChannel, default=0, const [0,445]}
       heel_r_ctrl.Scale.X {type=MovieSceneScriptingFloatChannel, default=1, const [1,445]}
       heel_r_ctrl.Scale.Y {type=MovieSceneScriptingFloatChannel, default=1, const [1,445]}
       heel_r_ctrl.Scale.Z {type=MovieSceneScriptingFloatChannel, default=1, const [1,445]}
       leg_r_softness {type=MovieSceneScriptingFloatChannel, default=0, const [0,753]}
       leg_r_pv_ik_ctrl.Location.X {
        type=MovieSceneScriptingFloatChannel
        default=0
        thinned_from=751
        keys [[-2,-5.6082],[-1,-5.4606],[0,-5.2401],[1,-5.1245],[4,-4.7335],[6,-4.4369],[7,-3.445],[8,-4.1104],[12,-3.4133],[13,-3.2301],[14,-3.0302],[15,-2.8058],[16,-2.5685],[17,-2.3243],[18,-2.0676],[19,-1.8184],[20,-1.5735],[21,-1.3422],[22,-1.1139],[23,-0.8766],[24,-0.6591],[25,-0.456],[26,-0.2692],[27,-0.1068],[28,0.0273],[29,0.1342],[30,0.21],[31,0.2472],[32,0.252],[33,0.2122],[34,0.1412],[35,0.0403],[36,-0.0953],[37,-0.2507],[38,-0.4315],[39,-0.632],[40,-0.8472],[41,-1.0732],[43,-1.5623],[45,-2.0696],[47,-2.5688],[48,-2.8081],[49,-3.0381],[50,-3.2574],[51,-3.4616],[52,-3.6428],[53,-3.8094],[54,-3.9472],[55,-4.069],[56,-4.1805],[57,-4.2795],[59,-4.4511],[61,-4.582],[64,-4.7148],[66,-4.7608],[68,-4.7819],[70,-4.7785],[74,-4.7214],[76,-4.6755],[78,-4.6097],[80,-4.5177],[81,-4.4525],[82,-4.3647],[83,-4.2564],[84,-4.1148],[86,-3.7521],[87,-3.6132],[88,-3.5199],[89,-3.4771],[90,-3.4859],[91,-3.5398],[92,-3.6325],[93,-3.754],[95,-4.0142],[96,-4.1237],[97,-4.2031],[98,-4.237],[99,-4.2004],[102,-3.9967],[103,-3.9385],[104,-3.893],[105,-3.8578],[107,-3.8125],[113,-3.7288],[116,-3.6657],[118,-3.606],[119,-3.5634],[120,-3.5107],[121,-3.4436],[122,-3.3672],[123,-3.2806],[124,-3.1806],[127,-2.8476],[128,-2.7437],[129,-2.6455],[130,-2.5601],[131,-2.489],[132,-2.4361],[133,-2.4006],[134,-2.3865],[135,-2.3909],[145,-2.4886],[153,-2.5255],[158,-2.523],[160,-2.5086],[162,-2.4794],[164,-2.4359],[166,-2.3797],[170,-2.2414],[177,-1.9821],[182,-1.8246],[184,-1.7671],[187,-1.6911],[190,-1.6281],[193,-1.5797],[195,-1.5565],[197,-1.5421],[199,-1.5362],[203,-1.5397],[208,-1.5365],[210,-1.5434],[211,-1.5502],[212,-1.5603],[213,-1.5744],[214,-1.6006],[215,-1.6444],[216,-1.7054],[217,-1.7821],[218,-1.8738],[219,-1.9798],[220,-2.099],[221,-2.2286],[222,-2.3672],[225,-2.807],[226,-2.9491],[227,-3.0811],[232,-3.6925],[234,-3.9196],[235,-4.0704],[236,-4.286],[237,-4.5581],[238,-4.8756],[239,-5.2336],[240,-5.6266],[241,-6.0522],[242,-6.5001],[243,-6.9688],[246,-8.4227],[248,-9.3562],[250,-10.2309],[252,-11.0417],[253,-11.4135],[254,-11.7548],[255,-12.064],[256,-12.3368],[257,-12.5613],[258,-12.7376],[259,-12.8646],[260,-12.9265],[261,-12.9271],[262,-12.8547],[263,-12.7058],[264,-12.4779],[268,-11.4421],[271,-10.7445],[273,-10.3361],[275,-9.9783],[277,-9.6716],[279,-9.4223],[281,-9.2324],[283,-9.108],[284,-9.0721],[285,-9.0548],[286,-9.0566],[294,-9.2331],[303,-9.2323],[309,-9.1285],[323,-5.6082],[324,-5.4606],[325,-5.2401],[326,-5.1245],[329,-4.7335],[331,-4.4369],[332,-3.445],[333,-4.1104],[337,-3.4133],[338,-3.2301],[339,-3.0302],[340,-2.8058],[341,-2.5685],[342,-2.3243],[343,-2.0676],[344,-1.8184],[345,-1.5735],[346,-1.3422],[347,-1.1139],[348,-0.8766],[349,-0.6591],[350,-0.456],[351,-0.2692],[352,-0.1068],[353,0.0273],[354,0.1342],[355,0.21],[356,0.2472],[357,0.252],[358,0.2122],[359,0.1412],[360,0.0403],[361,-0.0953],[362,-0.2507],[363,-0.4315],[364,-0.632],[365,-0.8472],[366,-1.0732],[368,-1.5623],[370,-2.0696],[372,-2.5688],[373,-2.8081],[374,-3.0381],[375,-3.2574],[376,-3.4616],[377,-3.6428],[378,-3.8094],[379,-3.9472],[380,-4.069],[381,-4.1805],[382,-4.2795],[384,-4.4511],[386,-4.582],[389,-4.7148],[391,-4.7608],[393,-4.7819],[395,-4.7785],[399,-4.7214],[401,-4.6755],[403,-4.6097],[405,-4.5177],[406,-4.4525],[407,-4.3647],[408,-4.2564],[409,-4.1148],[411,-3.7521],[412,-3.6132],[413,-3.5199],[414,-3.4771],[415,-3.4859],[416,-3.5398],[417,-3.6325],[418,-3.754],[420,-4.0142],[421,-4.1237],[422,-4.2031],[423,-4.237],[424,-4.2004],[427,-3.9967],[428,-3.9385],[429,-3.893],[430,-3.8578],[432,-3.8125],[438,-3.7288],[441,-3.6657],[443,-3.606],[444,-3.5634],[445,-3.5107],[446,-3.4436],[447,-3.3672],[448,-3.2806],[449,-3.1806],[452,-2.8476],[453,-2.7437],[454,-2.6455],[455,-2.5601],[456,-2.489],[457,-2.4361],[458,-2.4006],[459,-2.3865],[460,-2.3909],[470,-2.4886],[478,-2.5255],[483,-2.523],[485,-2.5086],[487,-2.4794],[489,-2.4359],[491,-2.3797],[495,-2.2414],[502,-1.9821],[507,-1.8246],[509,-1.7671],[512,-1.6911],[515,-1.6281],[518,-1.5797],[520,-1.5565],[522,-1.5421],[524,-1.5362],[528,-1.5397],[533,-1.5365],[535,-1.5434],[536,-1.5502],[537,-1.5603],[538,-1.5744],[539,-1.6006],[540,-1.6444],[541,-1.7054],[542,-1.7821],[543,-1.8738],[544,-1.9798],[545,-2.099],[546,-2.2286],[547,-2.3672],[550,-2.807],[551,-2.9491],[552,-3.0811],[557,-3.6925],[559,-3.9196],[560,-4.0704],[561,-4.286],[562,-4.5581],[563,-4.8756],[564,-5.2336],[565,-5.6266],[566,-6.0522],[567,-6.5001],[568,-6.9688],[571,-8.4227],[573,-9.3562],[575,-10.2309],[577,-11.0417],[578,-11.4135],[579,-11.7548],[580,-12.064],[581,-12.3368],[582,-12.5613],[583,-12.7376],[584,-12.8646],[585,-12.9265],[586,-12.9271],[587,-12.8547],[588,-12.7058],[589,-12.4779],[593,-11.4421],[596,-10.7445],[598,-10.3361],[600,-9.9783],[602,-9.6716],[604,-9.4223],[606,-9.2324],[608,-9.108],[609,-9.0721],[610,-9.0548],[611,-9.0566],[619,-9.2331],[628,-9.2323],[634,-9.1285],[676,-5.6082],[677,-5.4606],[678,-5.2401],[679,-5.1245],[682,-4.7335],[684,-4.4369],[685,-3.445],[686,-4.1104],[690,-3.4133],[691,-3.2301],[692,-3.0302],[693,-2.8058],[694,-2.5685],[695,-2.3243],[696,-2.0676],[697,-1.8184],[698,-1.5735],[699,-1.3422],[700,-1.1139],[701,-0.8766],[702,-0.6591],[703,-0.456],[704,-0.2692],[705,-0.1068],[706,0.0273],[707,0.1342],[708,0.21],[709,0.2472],[710,0.252],[711,0.2122],[712,0.1412],[713,0.0403],[714,-0.0953],[715,-0.2507],[716,-0.4315],[717,-0.632],[718,-0.8472],[719,-1.0732],[721,-1.5623],[723,-2.0696],[725,-2.5688],[726,-2.8081],[727,-3.0381],[728,-3.2574],[729,-3.4616],[730,-3.6428],[731,-3.8094],[732,-3.9472],[733,-4.069],[734,-4.1805],[735,-4.2795],[737,-4.4511],[739,-4.582],[742,-4.7148],[744,-4.7608],[746,-4.7819],[748,-4.7785],[752,-4.7214],[754,-4.6755],[755,-4.6448],[757,-4.5677],[759,-4.4525],[770,-9.0548],[771,-9.0566],[779,-9.2331],[788,-9.2323],[798,-9.0261],[803,-8.8632],[837,5.489],[887,-9.1063],[895,-8.8632],[929,5.489]]
       }
       leg_r_pv_ik_ctrl.Location.Y {
        type=MovieSceneScriptingFloatChannel
        default=0
        thinned_from=751
        keys [[-2,11.3236],[-1,11.1976],[0,11.0272],[6,10.6412],[7,5.6653],[8,10.5314],[11,10.4172],[13,10.3888],[18,10.3527],[22,10.4871],[33,10.5458],[51,10.8826],[68,10.8345],[75,10.7041],[78,10.5761],[80,10.4479],[81,10.3521],[82,10.2174],[83,10.0476],[84,9.8443],[85,9.6113],[86,9.3442],[87,9.0545],[88,8.7405],[89,8.4047],[91,7.6782],[95,6.1148],[96,5.7327],[97,5.3663],[98,5.023],[99,4.6974],[100,4.421],[101,4.1879],[102,4.0071],[103,3.8862],[104,3.8381],[105,3.8678],[106,3.9636],[107,4.1082],[108,4.2942],[109,4.5154],[110,4.7632],[111,5.0263],[116,6.4211],[118,6.9421],[120,7.4043],[121,7.606],[122,7.7886],[123,7.9471],[124,8.0773],[125,8.181],[128,8.4112],[131,8.5891],[136,8.7871],[143,8.9257],[155,8.9456],[160,8.8729],[163,8.77],[168,8.4964],[186,7.2763],[191,7.0253],[195,6.897],[197,6.8626],[200,6.8557],[208,6.9784],[212,7.1002],[215,7.2486],[217,7.3826],[219,7.5547],[221,7.7636],[223,8.0195],[226,8.4899],[235,10.1126],[236,10.2725],[237,10.4103],[238,10.5252],[239,10.6165],[240,10.6812],[241,10.718],[242,10.7258],[243,10.7031],[244,10.647],[245,10.5585],[246,10.434],[247,10.2747],[248,10.0756],[250,9.6168],[258,7.5299],[260,7.0401],[261,6.8167],[262,6.6137],[263,6.4363],[264,6.2846],[269,5.6156],[272,5.2434],[275,4.9065],[277,4.7077],[279,4.5359],[281,4.3922],[282,4.3317],[284,4.2369],[285,4.2026],[286,4.177],[295,4.038],[309,3.7179],[323,11.3236],[324,11.1976],[325,11.0272],[331,10.6412],[332,5.6653],[333,10.5314],[336,10.4172],[338,10.3888],[343,10.3527],[347,10.4871],[358,10.5458],[376,10.8826],[393,10.8345],[400,10.7041],[403,10.5761],[405,10.4479],[406,10.3521],[407,10.2174],[408,10.0476],[409,9.8443],[410,9.6113],[411,9.3442],[412,9.0545],[413,8.7405],[414,8.4047],[416,7.6782],[420,6.1148],[421,5.7327],[422,5.3663],[423,5.023],[424,4.6974],[425,4.421],[426,4.1879],[427,4.0071],[428,3.8862],[429,3.8381],[430,3.8678],[431,3.9636],[432,4.1082],[433,4.2942],[434,4.5154],[435,4.7632],[436,5.0263],[441,6.4211],[443,6.9421],[445,7.4043],[446,7.606],[447,7.7886],[448,7.9471],[449,8.0773],[450,8.181],[453,8.4112],[456,8.5891],[461,8.7871],[468,8.9257],[480,8.9456],[485,8.8729],[488,8.77],[493,8.4964],[511,7.2763],[516,7.0253],[520,6.897],[522,6.8626],[525,6.8557],[533,6.9784],[537,7.1002],[540,7.2486],[542,7.3826],[544,7.5547],[546,7.7636],[548,8.0195],[551,8.4899],[560,10.1126],[561,10.2725],[562,10.4103],[563,10.5252],[564,10.6165],[565,10.6812],[566,10.718],[567,10.7258],[568,10.7031],[569,10.647],[570,10.5585],[571,10.434],[572,10.2747],[573,10.0756],[575,9.6168],[583,7.5299],[585,7.0401],[586,6.8167],[587,6.6137],[588,6.4363],[589,6.2846],[594,5.6156],[597,5.2434],[600,4.9065],[602,4.7077],[604,4.5359],[606,4.3922],[607,4.3317],[609,4.2369],[610,4.2026],[611,4.177],[620,4.038],[634,3.7179],[676,11.3236],[677,11.1976],[678,11.0272],[684,10.6412],[685,5.6653],[686,10.5314],[689,10.4172],[691,10.3888],[696,10.3527],[700,10.4871],[711,10.5458],[729,10.8826],[746,10.8345],[753,10.7041],[755,10.6268],[757,10.517],[759,10.3521],[770,4.2026],[771,4.177],[780,4.038],[803,3.5286],[837,6.0097],[887,3.6951],[895,3.5286],[929,6.0097]]
       }
       leg_r_pv_ik_ctrl.Location.Z {
        type=MovieSceneScriptingFloatChannel
        default=0
        thinned_from=751
        keys [[-2,14.7111],[-1,14.4703],[0,14.1496],[6,13.4489],[7,6.0017],[8,13.2598],[10,13.1197],[12,13.0497],[18,13.0665],[22,13.41],[55,13.9314],[68,13.794],[74,13.6092],[76,13.499],[78,13.3312],[80,13.1057],[81,12.9388],[82,12.7065],[83,12.4174],[84,12.0768],[85,11.6945],[86,11.2645],[87,10.8062],[88,10.3184],[90,9.2742],[94,7.0899],[95,6.5665],[96,6.0688],[97,5.6025],[98,5.1749],[99,4.778],[100,4.4472],[101,4.1723],[102,3.9619],[103,3.8226],[104,3.768],[105,3.8032],[106,3.915],[107,4.0845],[108,4.304],[109,4.5679],[110,4.8672],[111,5.1896],[113,5.8857],[117,7.3401],[118,7.6899],[120,8.3395],[121,8.63],[122,8.897],[123,9.1325],[124,9.3293],[125,9.489],[128,9.8544],[131,10.1433],[134,10.3529],[139,10.5716],[147,10.7119],[155,10.6944],[159,10.6155],[162,10.4866],[166,10.1949],[184,8.4427],[189,8.0542],[193,7.834],[195,7.7603],[197,7.7143],[200,7.7054],[208,7.8749],[211,7.9914],[213,8.1005],[216,8.3253],[218,8.5273],[220,8.7828],[221,8.9296],[222,9.0968],[223,9.2838],[225,9.7202],[227,10.2173],[231,11.3322],[235,12.5323],[238,13.3381],[240,13.8063],[242,14.2004],[244,14.5045],[245,14.6173],[246,14.6965],[247,14.7416],[253,14.7265],[262,15.0829],[279,16.1145],[285,16.2894],[294,16.1936],[309,15.5084],[323,14.7111],[324,14.4703],[325,14.1496],[331,13.4489],[332,6.0017],[333,13.2598],[335,13.1197],[337,13.0497],[343,13.0665],[347,13.41],[380,13.9314],[393,13.794],[399,13.6092],[401,13.499],[403,13.3312],[405,13.1057],[406,12.9388],[407,12.7065],[408,12.4174],[409,12.0768],[410,11.6945],[411,11.2645],[412,10.8062],[413,10.3184],[415,9.2742],[419,7.0899],[420,6.5665],[421,6.0688],[422,5.6025],[423,5.1749],[424,4.778],[425,4.4472],[426,4.1723],[427,3.9619],[428,3.8226],[429,3.768],[430,3.8032],[431,3.915],[432,4.0845],[433,4.304],[434,4.5679],[435,4.8672],[436,5.1896],[438,5.8857],[442,7.3401],[443,7.6899],[445,8.3395],[446,8.63],[447,8.897],[448,9.1325],[449,9.3293],[450,9.489],[453,9.8544],[456,10.1433],[459,10.3529],[464,10.5716],[472,10.7119],[480,10.6944],[484,10.6155],[487,10.4866],[491,10.1949],[509,8.4427],[514,8.0542],[518,7.834],[520,7.7603],[522,7.7143],[525,7.7054],[533,7.8749],[536,7.9914],[538,8.1005],[541,8.3253],[543,8.5273],[545,8.7828],[546,8.9296],[547,9.0968],[548,9.2838],[550,9.7202],[552,10.2173],[556,11.3322],[560,12.5323],[563,13.3381],[565,13.8063],[567,14.2004],[569,14.5045],[570,14.6173],[571,14.6965],[572,14.7416],[578,14.7265],[587,15.0829],[604,16.1145],[610,16.2894],[619,16.1936],[634,15.5084],[676,14.7111],[677,14.4703],[678,14.1496],[684,13.4489],[685,6.0017],[686,13.2598],[688,13.1197],[690,13.0497],[696,13.0665],[700,13.41],[733,13.9314],[746,13.794],[752,13.6092],[754,13.499],[757,13.2269],[759,12.9388],[770,16.2894],[779,16.1936],[803,15.0213],[837,15.415],[887,15.4546],[895,15.0213],[929,15.415]]
       }
       leg_r_pv_ik_ctrl.Rotation.X {
        type=MovieSceneScriptingFloatChannel
        default=0
        thinned_from=751
        keys [[-2,-18.1634],[-1,-17.8955],[0,-17.5376],[6,-16.7497],[7,-8.0794],[8,-16.5355],[10,-16.3759],[12,-16.2952],[18,-16.3076],[22,-16.6897],[56,-17.2892],[72,-17.0154],[76,-16.8072],[79,-16.5008],[80,-16.364],[81,-16.1754],[82,-15.9122],[83,-15.5836],[84,-15.1951],[85,-14.7569],[86,-14.2617],[87,-13.7312],[88,-13.1633],[89,-12.5631],[91,-11.2897],[94,-9.3104],[95,-8.6681],[96,-8.0521],[97,-7.4699],[98,-6.9317],[99,-6.4279],[100,-6.0049],[101,-5.6513],[102,-5.3792],[103,-5.1984],[104,-5.1272],[105,-5.1727],[106,-5.3175],[107,-5.5363],[108,-5.819],[109,-6.1571],[110,-6.5386],[111,-6.9472],[113,-7.8214],[116,-9.176],[118,-10.0394],[120,-10.8222],[121,-11.1696],[122,-11.4876],[123,-11.7669],[124,-11.9996],[125,-12.1876],[128,-12.6161],[131,-12.9532],[134,-13.1976],[139,-13.4525],[147,-13.6159],[156,-13.5786],[160,-13.4649],[163,-13.282],[168,-12.8016],[184,-10.9363],[189,-10.4687],[193,-10.2025],[196,-10.0829],[199,-10.0416],[207,-10.2174],[211,-10.3925],[213,-10.5241],[216,-10.7952],[218,-11.0381],[220,-11.3446],[221,-11.5201],[223,-11.942],[225,-12.4586],[227,-13.0429],[231,-14.3387],[235,-15.7132],[237,-16.3348],[239,-16.8895],[241,-17.3631],[243,-17.7415],[245,-18.0023],[246,-18.0796],[247,-18.1181],[253,-18.0158],[262,-18.2417],[281,-19.1358],[288,-19.2248],[299,-18.9144],[309,-18.3557],[323,-18.1634],[324,-17.8955],[325,-17.5376],[331,-16.7497],[332,-8.0794],[333,-16.5355],[335,-16.3759],[337,-16.2952],[343,-16.3076],[347,-16.6897],[381,-17.2892],[397,-17.0154],[401,-16.8072],[404,-16.5008],[405,-16.364],[406,-16.1754],[407,-15.9122],[408,-15.5836],[409,-15.1951],[410,-14.7569],[411,-14.2617],[412,-13.7312],[413,-13.1633],[414,-12.5631],[416,-11.2897],[419,-9.3104],[420,-8.6681],[421,-8.0521],[422,-7.4699],[423,-6.9317],[424,-6.4279],[425,-6.0049],[426,-5.6513],[427,-5.3792],[428,-5.1984],[429,-5.1272],[430,-5.1727],[431,-5.3175],[432,-5.5363],[433,-5.819],[434,-6.1571],[435,-6.5386],[436,-6.9472],[438,-7.8214],[441,-9.176],[443,-10.0394],[445,-10.8222],[446,-11.1696],[447,-11.4876],[448,-11.7669],[449,-11.9996],[450,-12.1876],[453,-12.6161],[456,-12.9532],[459,-13.1976],[464,-13.4525],[472,-13.6159],[481,-13.5786],[485,-13.4649],[488,-13.282],[493,-12.8016],[509,-10.9363],[514,-10.4687],[518,-10.2025],[521,-10.0829],[524,-10.0416],[532,-10.2174],[536,-10.3925],[538,-10.5241],[541,-10.7952],[543,-11.0381],[545,-11.3446],[546,-11.5201],[548,-11.942],[550,-12.4586],[552,-13.0429],[556,-14.3387],[560,-15.7132],[562,-16.3348],[564,-16.8895],[566,-17.3631],[568,-17.7415],[570,-18.0023],[571,-18.0796],[572,-18.1181],[578,-18.0158],[587,-18.2417],[606,-19.1358],[613,-19.2248],[624,-18.9144],[634,-18.3557],[676,-18.1634],[677,-17.8955],[678,-17.5376],[684,-16.7497],[685,-8.0794],[686,-16.5355],[688,-16.3759],[690,-16.2952],[696,-16.3076],[700,-16.6897],[734,-17.2892],[750,-17.0154],[754,-16.8072],[757,-16.5008],[759,-16.1754],[770,-19.2178],[780,-19.0785],[803,-17.8142],[837,-18.1634],[887,-18.2961],[895,-17.8142],[929,-18.1634]]
       }
       leg_r_pv_ik_ctrl.Rotation.Y {
        type=MovieSceneScriptingFloatChannel
        default=0
        thinned_from=751
        keys [[-2,0.0812],[-1,0.0795],[0,0.0785],[1,0.0807],[2,0.0826],[3,0.0848],[4,0.0872],[5,0.0904],[6,0.0944],[7,-1.3056],[8,0.1046],[9,0.1108],[10,0.1176],[11,0.1252],[12,0.1335],[13,0.1428],[15,0.1651],[16,0.1771],[17,0.1896],[18,0.2028],[20,0.2336],[21,0.2482],[22,0.262],[23,0.2743],[24,0.2856],[25,0.2961],[26,0.3059],[27,0.3145],[28,0.3218],[29,0.3278],[30,0.3322],[31,0.3349],[32,0.3359],[33,0.3348],[34,0.3323],[35,0.3283],[36,0.3226],[37,0.316],[38,0.3081],[39,0.2993],[41,0.2793],[43,0.2566],[47,0.2084],[48,0.1966],[50,0.1742],[51,0.1639],[52,0.1546],[53,0.1462],[54,0.139],[55,0.1328],[56,0.1269],[57,0.1216],[59,0.1123],[60,0.1085],[61,0.1045],[62,0.1015],[64,0.096],[65,0.0939],[66,0.0921],[67,0.0905],[69,0.0881],[71,0.0868],[73,0.0859],[75,0.0844],[76,0.0834],[78,0.0804],[80,0.0767],[81,0.0739],[82,0.0698],[84,0.06],[85,0.056],[86,0.0505],[87,0.0421],[88,0.0306],[89,0.0159],[90,-0.0018],[91,-0.022],[92,-0.0442],[94,-0.0912],[95,-0.1137],[96,-0.1341],[97,-0.1516],[98,-0.1649],[99,-0.1729],[100,-0.1767],[101,-0.1792],[102,-0.1801],[103,-0.1797],[104,-0.1781],[105,-0.1751],[106,-0.1709],[107,-0.1657],[108,-0.1595],[109,-0.1522],[110,-0.1442],[111,-0.1351],[112,-0.1255],[113,-0.1153],[114,-0.1043],[115,-0.093],[116,-0.0814],[117,-0.0695],[119,-0.045],[122,-0.0071],[123,0.0051],[124,0.0168],[125,0.0279],[126,0.0383],[127,0.0481],[128,0.0574],[129,0.066],[130,0.0737],[131,0.0803],[132,0.0856],[133,0.0897],[134,0.0925],[135,0.094],[137,0.0959],[140,0.0976],[141,0.098],[143,0.0984],[147,0.0983],[154,0.0968],[157,0.0954],[161,0.0929],[163,0.0911],[178,0.0734],[183,0.0673],[186,0.0641],[189,0.0614],[192,0.0594],[195,0.0581],[197,0.0576],[199,0.0575],[201,0.0579],[203,0.0589],[208,0.0623],[213,0.0666],[214,0.0672],[215,0.0669],[216,0.0661],[217,0.0647],[218,0.063],[219,0.0609],[222,0.0537],[223,0.0517],[224,0.0503],[225,0.0496],[226,0.0498],[227,0.051],[228,0.0532],[229,0.0558],[230,0.059],[231,0.0625],[232,0.0669],[234,0.0765],[235,0.0792],[236,0.0775],[237,0.0709],[238,0.0601],[239,0.0452],[240,0.0264],[241,0.0036],[242,-0.0229],[243,-0.053],[244,-0.0867],[245,-0.1233],[246,-0.1631],[247,-0.2052],[248,-0.2504],[249,-0.2967],[252,-0.4388],[253,-0.4856],[254,-0.5309],[255,-0.5749],[256,-0.617],[257,-0.6562],[258,-0.6922],[259,-0.7251],[260,-0.7532],[261,-0.7767],[262,-0.7944],[263,-0.8056],[264,-0.8102],[275,-0.83],[306,-0.9061],[309,-0.9101],[323,0.0812],[324,0.0795],[325,0.0785],[326,0.0807],[327,0.0826],[328,0.0848],[329,0.0872],[330,0.0904],[331,0.0944],[332,-1.3056],[333,0.1046],[334,0.1108],[335,0.1176],[336,0.1252],[337,0.1335],[338,0.1428],[340,0.1651],[341,0.1771],[342,0.1896],[343,0.2028],[345,0.2336],[346,0.2482],[347,0.262],[348,0.2743],[349,0.2856],[350,0.2961],[351,0.3059],[352,0.3145],[353,0.3218],[354,0.3278],[355,0.3322],[356,0.3349],[357,0.3359],[358,0.3348],[359,0.3323],[360,0.3283],[361,0.3226],[362,0.316],[363,0.3081],[364,0.2993],[366,0.2793],[368,0.2566],[372,0.2084],[373,0.1966],[375,0.1742],[376,0.1639],[377,0.1546],[378,0.1462],[379,0.139],[380,0.1328],[381,0.1269],[382,0.1216],[384,0.1123],[385,0.1085],[386,0.1045],[387,0.1015],[389,0.096],[390,0.0939],[391,0.0921],[392,0.0905],[394,0.0881],[396,0.0868],[398,0.0859],[400,0.0844],[401,0.0834],[403,0.0804],[405,0.0767],[406,0.0739],[407,0.0698],[409,0.06],[410,0.056],[411,0.0505],[412,0.0421],[413,0.0306],[414,0.0159],[415,-0.0018],[416,-0.022],[417,-0.0442],[419,-0.0912],[420,-0.1137],[421,-0.1341],[422,-0.1516],[423,-0.1649],[424,-0.1729],[425,-0.1767],[426,-0.1792],[427,-0.1801],[428,-0.1797],[429,-0.1781],[430,-0.1751],[431,-0.1709],[432,-0.1657],[433,-0.1595],[434,-0.1522],[435,-0.1442],[436,-0.1351],[437,-0.1255],[438,-0.1153],[439,-0.1043],[440,-0.093],[441,-0.0814],[442,-0.0695],[444,-0.045],[447,-0.0071],[448,0.0051],[449,0.0168],[450,0.0279],[451,0.0383],[452,0.0481],[453,0.0574],[454,0.066],[455,0.0737],[456,0.0803],[457,0.0856],[458,0.0897],[459,0.0925],[460,0.094],[462,0.0959],[465,0.0976],[466,0.098],[468,0.0984],[472,0.0983],[479,0.0968],[482,0.0954],[486,0.0929],[488,0.0911],[503,0.0734],[508,0.0673],[511,0.0641],[514,0.0614],[517,0.0594],[520,0.0581],[522,0.0576],[524,0.0575],[526,0.0579],[528,0.0589],[533,0.0623],[538,0.0666],[539,0.0672],[540,0.0669],[541,0.0661],[542,0.0647],[543,0.063],[544,0.0609],[547,0.0537],[548,0.0517],[549,0.0503],[550,0.0496],[551,0.0498],[552,0.051],[553,0.0532],[554,0.0558],[555,0.059],[556,0.0625],[557,0.0669],[559,0.0765],[560,0.0792],[561,0.0775],[562,0.0709],[563,0.0601],[564,0.0452],[565,0.0264],[566,0.0036],[567,-0.0229],[568,-0.053],[569,-0.0867],[570,-0.1233],[571,-0.1631],[572,-0.2052],[573,-0.2504],[574,-0.2967],[577,-0.4388],[578,-0.4856],[579,-0.5309],[580,-0.5749],[581,-0.617],[582,-0.6562],[583,-0.6922],[584,-0.7251],[585,-0.7532],[586,-0.7767],[587,-0.7944],[588,-0.8056],[589,-0.8102],[600,-0.83],[631,-0.9061],[634,-0.9101],[676,0.0812],[677,0.0795],[678,0.0785],[679,0.0807],[680,0.0826],[681,0.0848],[682,0.0872],[683,0.0904],[684,0.0944],[685,-1.3056],[686,0.1046],[687,0.1108],[688,0.1176],[689,0.1252],[690,0.1335],[691,0.1428],[693,0.1651],[694,0.1771],[695,0.1896],[696,0.2028],[698,0.2336],[699,0.2482],[700,0.262],[701,0.2743],[702,0.2856],[703,0.2961],[704,0.3059],[705,0.3145],[706,0.3218],[707,0.3278],[708,0.3322],[709,0.3349],[710,0.3359],[711,0.3348],[712,0.3323],[713,0.3283],[714,0.3226],[715,0.316],[716,0.3081],[717,0.2993],[719,0.2793],[721,0.2566],[725,0.2084],[726,0.1966],[728,0.1742],[729,0.1639],[730,0.1546],[731,0.1462],[732,0.139],[733,0.1328],[734,0.1269],[735,0.1216],[737,0.1123],[738,0.1085],[739,0.1045],[740,0.1015],[742,0.096],[743,0.0939],[744,0.0921],[745,0.0905],[747,0.0881],[749,0.0868],[751,0.0859],[753,0.0844],[754,0.0834],[755,0.082],[757,0.0787],[759,0.0739],[770,-0.857],[793,-0.9089],[803,-0.915],[837,0.0812],[887,-0.9112],[895,-0.915],[929,0.0812]]
       }
       leg_r_pv_ik_ctrl.Rotation.Z {
        type=MovieSceneScriptingFloatChannel
        default=0
        thinned_from=751
        keys [[-2,6.2572],[-1,6.1084],[0,5.8834],[1,5.7624],[4,5.3522],[6,5.0373],[7,3.2651],[8,4.687],[12,3.9296],[13,3.7288],[14,3.509],[15,3.2629],[16,3.0022],[17,2.7339],[18,2.4516],[19,2.1769],[20,1.9072],[21,1.6528],[22,1.4021],[23,1.1415],[24,0.9026],[25,0.6797],[26,0.4747],[27,0.2965],[28,0.1494],[29,0.0323],[30,-0.0506],[31,-0.0912],[32,-0.096],[33,-0.0521],[34,0.0261],[35,0.1371],[36,0.2861],[37,0.4567],[38,0.6549],[39,0.8747],[40,1.1102],[41,1.3574],[43,1.8919],[45,2.4455],[47,2.9894],[48,3.25],[49,3.5003],[50,3.7389],[51,3.9609],[52,4.1579],[53,4.3389],[54,4.4886],[55,4.6207],[56,4.7419],[57,4.8495],[59,5.0362],[61,5.1789],[64,5.3245],[66,5.3759],[68,5.4005],[70,5.3987],[74,5.3416],[77,5.2652],[79,5.1883],[80,5.1381],[81,5.0728],[82,4.9849],[83,4.8763],[84,4.732],[86,4.3569],[87,4.2166],[88,4.1276],[89,4.0964],[90,4.125],[91,4.2072],[92,4.3366],[93,4.5023],[95,4.8627],[96,5.0219],[97,5.1463],[98,5.2156],[99,5.1988],[103,4.9456],[105,4.8471],[113,4.5603],[117,4.3804],[118,4.3267],[119,4.2629],[120,4.189],[121,4.1002],[122,4.0024],[123,3.8949],[125,3.6455],[127,3.3823],[128,3.2613],[129,3.1472],[130,3.0481],[131,2.9657],[132,2.9042],[133,2.8626],[134,2.8456],[135,2.8495],[146,2.962],[153,2.9958],[158,2.9943],[160,2.9792],[162,2.9482],[164,2.9014],[166,2.8408],[170,2.6903],[177,2.405],[182,2.2302],[184,2.1659],[187,2.0807],[190,2.0097],[193,1.9549],[195,1.9286],[197,1.9121],[199,1.9053],[208,1.9038],[210,1.9112],[212,1.93],[213,1.9458],[214,1.9756],[215,2.0258],[216,2.0958],[217,2.1836],[218,2.2885],[219,2.4094],[220,2.5449],[221,2.6916],[223,3.0096],[225,3.3366],[226,3.492],[227,3.6346],[230,4.029],[233,4.3971],[234,4.5106],[235,4.6673],[236,4.9081],[237,5.2359],[238,5.637],[239,6.1047],[240,6.6318],[241,7.2148],[242,7.8415],[243,8.5097],[244,9.2124],[247,11.4195],[252,15.1502],[254,16.5605],[256,17.8581],[257,18.4429],[258,18.9793],[259,19.4651],[260,19.8811],[261,20.2305],[262,20.4986],[263,20.6799],[264,20.7708],[291,22.0796],[309,22.1513],[323,6.2572],[324,6.1084],[325,5.8834],[326,5.7624],[329,5.3522],[331,5.0373],[332,3.2651],[333,4.687],[337,3.9296],[338,3.7288],[339,3.509],[340,3.2629],[341,3.0022],[342,2.7339],[343,2.4516],[344,2.1769],[345,1.9072],[346,1.6528],[347,1.4021],[348,1.1415],[349,0.9026],[350,0.6797],[351,0.4747],[352,0.2965],[353,0.1494],[354,0.0323],[355,-0.0506],[356,-0.0912],[357,-0.096],[358,-0.0521],[359,0.0261],[360,0.1371],[361,0.2861],[362,0.4567],[363,0.6549],[364,0.8747],[365,1.1102],[366,1.3574],[368,1.8919],[370,2.4455],[372,2.9894],[373,3.25],[374,3.5003],[375,3.7389],[376,3.9609],[377,4.1579],[378,4.3389],[379,4.4886],[380,4.6207],[381,4.7419],[382,4.8495],[384,5.0362],[386,5.1789],[389,5.3245],[391,5.3759],[393,5.4005],[395,5.3987],[399,5.3416],[402,5.2652],[404,5.1883],[405,5.1381],[406,5.0728],[407,4.9849],[408,4.8763],[409,4.732],[411,4.3569],[412,4.2166],[413,4.1276],[414,4.0964],[415,4.125],[416,4.2072],[417,4.3366],[418,4.5023],[420,4.8627],[421,5.0219],[422,5.1463],[423,5.2156],[424,5.1988],[428,4.9456],[430,4.8471],[438,4.5603],[442,4.3804],[443,4.3267],[444,4.2629],[445,4.189],[446,4.1002],[447,4.0024],[448,3.8949],[450,3.6455],[452,3.3823],[453,3.2613],[454,3.1472],[455,3.0481],[456,2.9657],[457,2.9042],[458,2.8626],[459,2.8456],[460,2.8495],[471,2.962],[478,2.9958],[483,2.9943],[485,2.9792],[487,2.9482],[489,2.9014],[491,2.8408],[495,2.6903],[502,2.405],[507,2.2302],[509,2.1659],[512,2.0807],[515,2.0097],[518,1.9549],[520,1.9286],[522,1.9121],[524,1.9053],[533,1.9038],[535,1.9112],[537,1.93],[538,1.9458],[539,1.9756],[540,2.0258],[541,2.0958],[542,2.1836],[543,2.2885],[544,2.4094],[545,2.5449],[546,2.6916],[548,3.0096],[550,3.3366],[551,3.492],[552,3.6346],[555,4.029],[558,4.3971],[559,4.5106],[560,4.6673],[561,4.9081],[562,5.2359],[563,5.637],[564,6.1047],[565,6.6318],[566,7.2148],[567,7.8415],[568,8.5097],[569,9.2124],[572,11.4195],[577,15.1502],[579,16.5605],[581,17.8581],[582,18.4429],[583,18.9793],[584,19.4651],[585,19.8811],[586,20.2305],[587,20.4986],[588,20.6799],[589,20.7708],[616,22.0796],[634,22.1513],[676,6.2572],[677,6.1084],[678,5.8834],[679,5.7624],[682,5.3522],[684,5.0373],[685,3.2651],[686,4.687],[690,3.9296],[691,3.7288],[692,3.509],[693,3.2629],[694,3.0022],[695,2.7339],[696,2.4516],[697,2.1769],[698,1.9072],[699,1.6528],[700,1.4021],[701,1.1415],[702,0.9026],[703,0.6797],[704,0.4747],[705,0.2965],[706,0.1494],[707,0.0323],[708,-0.0506],[709,-0.0912],[710,-0.096],[711,-0.0521],[712,0.0261],[713,0.1371],[714,0.2861],[715,0.4567],[716,0.6549],[717,0.8747],[718,1.1102],[719,1.3574],[721,1.8919],[723,2.4455],[725,2.9894],[726,3.25],[727,3.5003],[728,3.7389],[729,3.9609],[730,4.1579],[731,4.3389],[732,4.4886],[733,4.6207],[734,4.7419],[735,4.8495],[737,5.0362],[739,5.1789],[742,5.3245],[744,5.3759],[746,5.4005],[748,5.3987],[752,5.3416],[755,5.2652],[757,5.1883],[759,5.0728],[770,21.8761],[788,22.2086],[803,21.9455],[837,6.2572],[887,22.1364],[895,21.9455],[929,6.2572]]
       }
       leg_r_pv_ik_ctrl.Scale.X {type=MovieSceneScriptingFloatChannel, default=1, const [1,751]}
       leg_r_pv_ik_ctrl.Scale.Y {type=MovieSceneScriptingFloatChannel, default=1, const [1,751]}
       leg_r_pv_ik_ctrl.Scale.Z {type=MovieSceneScriptingFloatChannel, default=1, const [1,751]}
       upperarm_l_fk_ctrl.Location.X {type=MovieSceneScriptingFloatChannel, default=0, thinned_from=642, keys [[-2,0],[929,0]]}
       upperarm_l_fk_ctrl.Location.Y {type=MovieSceneScriptingFloatChannel, default=0, thinned_from=642, keys [[-2,0],[929,0]]}
       upperarm_l_fk_ctrl.Location.Z {type=MovieSceneScriptingFloatChannel, default=0, thinned_from=642, keys [[-2,0],[929,0]]}
       upperarm_l_fk_ctrl.Rotation.X {
        type=MovieSceneScriptingFloatChannel
        default=64.3382
        thinned_from=576
        keys [[-2,63.1588],[-1,63.3695],[6,63.0935],[82,61.0313],[84,60.7795],[87,59.9488],[90,58.5727],[95,55.7903],[97,54.9448],[99,54.3611],[101,54.0074],[103,53.9004],[106,54.1616],[118,57.0816],[121,57.3706],[125,57.203],[131,56.6091],[135,56.6933],[167,61.2017],[175,61.4688],[191,61.1004],[214,63.0629],[218,63.8625],[220,64.5001],[232,68.4822],[238,69.3431],[246,69.1231],[258,67.3257],[275,64.631],[286,64.961],[303,64.4368],[328,58.3055],[339,59.0246],[356,61.4706],[407,61.0313],[409,60.7795],[412,59.9488],[421,56.5204],[424,55.798],[427,55.4345],[431,55.5051],[444,57.2865],[448,57.3626],[456,56.6091],[460,56.6933],[492,61.2017],[500,61.4688],[516,61.1004],[539,63.0629],[543,63.8625],[545,64.5001],[557,68.4822],[563,69.3431],[571,69.1231],[583,67.3257],[608,63.7523],[635,63.7078],[647,29.7472],[676,3.407],[677,3.3956],[702,0.7922],[725,-0.7945],[734,-1.1647],[776,0.0402],[778,0.2169],[781,0.4226],[833,3.3024],[837,3.407],[887,68.6907],[897,64.5397],[903,63.5025],[929,63.1588]]
       }
       upperarm_l_fk_ctrl.Rotation.Y {
        type=MovieSceneScriptingFloatChannel
        default=-14.9498
        thinned_from=645
        keys [[-2,-10.5519],[-1,-10.2263],[3,-9.2824],[7,-8.2597],[8,-7.979],[9,-7.653],[10,-7.28],[11,-6.8688],[12,-6.425],[13,-5.9556],[15,-4.9749],[16,-4.4748],[17,-3.9881],[18,-3.5128],[19,-3.0548],[20,-2.6241],[21,-2.2262],[22,-1.8683],[23,-1.5581],[24,-1.302],[25,-1.0875],[26,-0.9044],[27,-0.7501],[28,-0.6296],[29,-0.5377],[30,-0.4708],[31,-0.4282],[32,-0.4043],[33,-0.3901],[34,-0.3794],[35,-0.3774],[36,-0.3792],[37,-0.3847],[38,-0.3959],[39,-0.4084],[40,-0.4229],[42,-0.4576],[43,-0.4789],[45,-0.516],[46,-0.5353],[48,-0.5704],[49,-0.586],[50,-0.6159],[51,-0.6223],[52,-0.6468],[53,-0.6737],[54,-0.6987],[55,-0.7285],[56,-0.7615],[57,-0.7899],[58,-0.826],[59,-0.8571],[60,-0.895],[61,-0.9259],[62,-0.9643],[63,-0.9987],[64,-1.0355],[65,-1.0687],[68,-1.1775],[71,-1.2789],[72,-1.3096],[74,-1.3767],[81,-1.5656],[82,-1.5641],[83,-1.5101],[84,-1.3997],[85,-1.2157],[86,-0.9651],[87,-0.6654],[88,-0.3089],[89,0.1041],[90,0.5503],[91,1.0079],[92,1.4496],[93,1.8536],[94,2.1976],[95,2.457],[96,2.6062],[97,2.6319],[98,2.5499],[99,2.3753],[100,2.1082],[101,1.7566],[102,1.3238],[103,0.8101],[104,0.2213],[105,-0.4198],[106,-1.0828],[108,-2.4425],[109,-3.1092],[110,-3.7557],[111,-4.374],[112,-4.9518],[114,-6.0566],[115,-6.5805],[116,-7.0742],[117,-7.5167],[118,-7.8933],[119,-8.1906],[120,-8.404],[121,-8.5441],[122,-8.6342],[123,-8.6668],[124,-8.6373],[125,-8.5555],[126,-8.4341],[130,-7.8411],[131,-7.719],[133,-7.5249],[137,-7.2512],[143,-6.9463],[161,-6.1706],[171,-5.9214],[173,-5.8379],[174,-5.786],[175,-5.716],[176,-5.6205],[180,-5.1572],[182,-4.9622],[183,-4.8868],[184,-4.8351],[188,-4.7003],[191,-4.6306],[193,-4.6091],[195,-4.6124],[197,-4.6423],[199,-4.7033],[201,-4.7881],[203,-4.8998],[204,-4.9734],[205,-5.0694],[206,-5.1842],[207,-5.3163],[209,-5.6197],[210,-5.7932],[211,-5.9821],[212,-6.1967],[213,-6.4457],[214,-6.7397],[215,-7.0806],[216,-7.4603],[217,-7.8832],[218,-8.372],[219,-8.941],[220,-9.6026],[221,-10.355],[222,-11.1744],[223,-12.0361],[226,-14.7172],[227,-15.5827],[228,-16.4034],[229,-17.1541],[230,-17.816],[231,-18.3683],[232,-18.8089],[233,-19.1416],[234,-19.3889],[235,-19.5778],[236,-19.7174],[237,-19.804],[238,-19.8413],[239,-19.8345],[241,-19.6928],[243,-19.4137],[245,-19.027],[249,-18.0613],[259,-15.2083],[261,-14.705],[262,-14.5074],[263,-14.3519],[264,-14.2463],[265,-14.1939],[266,-14.1874],[278,-15.1027],[282,-15.2931],[286,-15.3741],[291,-15.3062],[294,-15.1752],[296,-15.0188],[298,-14.758],[300,-14.4143],[302,-14.0008],[304,-13.7749],[308,-12.7816],[310,-12.2428],[323,-10.5519],[324,-10.2263],[328,-9.2824],[332,-8.2597],[333,-7.979],[334,-7.653],[335,-7.28],[336,-6.8688],[337,-6.425],[338,-5.9556],[340,-4.9749],[341,-4.4748],[342,-3.9881],[343,-3.5128],[344,-3.0548],[345,-2.6241],[346,-2.2262],[347,-1.8683],[348,-1.5581],[349,-1.302],[350,-1.0875],[351,-0.9044],[352,-0.7501],[353,-0.6296],[354,-0.5377],[355,-0.4708],[356,-0.4282],[357,-0.4043],[358,-0.3901],[359,-0.3794],[360,-0.3774],[361,-0.3792],[362,-0.3847],[363,-0.3959],[364,-0.4084],[365,-0.4229],[367,-0.4576],[368,-0.4789],[370,-0.516],[371,-0.5353],[373,-0.5704],[374,-0.586],[375,-0.6159],[376,-0.6223],[377,-0.6468],[378,-0.6737],[379,-0.6987],[380,-0.7285],[381,-0.7615],[382,-0.7899],[383,-0.826],[384,-0.8571],[385,-0.895],[386,-0.9259],[387,-0.9643],[388,-0.9987],[389,-1.0355],[390,-1.0687],[393,-1.1775],[396,-1.2789],[397,-1.3096],[399,-1.3767],[406,-1.5656],[407,-1.5641],[408,-1.5101],[409,-1.3997],[410,-1.2157],[411,-0.9651],[412,-0.6654],[413,-0.3089],[414,0.1041],[415,0.5503],[416,1.0079],[417,1.4496],[418,1.8536],[419,2.1976],[420,2.457],[421,2.6062],[422,2.6319],[423,2.5499],[424,2.3753],[425,2.1082],[426,1.7566],[427,1.3238],[428,0.8101],[429,0.2213],[430,-0.4198],[431,-1.0828],[433,-2.4425],[434,-3.1092],[435,-3.7557],[436,-4.374],[437,-4.9518],[439,-6.0566],[440,-6.5805],[441,-7.0742],[442,-7.5167],[443,-7.8933],[444,-8.1906],[445,-8.404],[446,-8.5441],[447,-8.6342],[448,-8.6668],[449,-8.6373],[450,-8.5555],[451,-8.4341],[455,-7.8411],[456,-7.719],[458,-7.5249],[462,-7.2512],[468,-6.9463],[486,-6.1706],[496,-5.9214],[498,-5.8379],[499,-5.786],[500,-5.716],[501,-5.6205],[505,-5.1572],[507,-4.9622],[508,-4.8868],[509,-4.8351],[513,-4.7003],[516,-4.6306],[518,-4.6091],[520,-4.6124],[522,-4.6423],[524,-4.7033],[526,-4.7881],[528,-4.8998],[529,-4.9734],[530,-5.0694],[531,-5.1842],[532,-5.3163],[534,-5.6197],[535,-5.7932],[536,-5.9821],[537,-6.1967],[538,-6.4457],[539,-6.7397],[540,-7.0806],[541,-7.4603],[542,-7.8832],[543,-8.372],[544,-8.941],[545,-9.6026],[546,-10.355],[547,-11.1744],[548,-12.0361],[551,-14.7172],[552,-15.5827],[553,-16.4034],[554,-17.1541],[555,-17.816],[556,-18.3683],[557,-18.8089],[558,-19.1416],[559,-19.3889],[560,-19.5778],[561,-19.7174],[562,-19.804],[563,-19.8413],[564,-19.8345],[566,-19.6928],[568,-19.4137],[570,-19.027],[574,-18.0613],[584,-15.2083],[586,-14.705],[587,-14.5074],[588,-14.3519],[589,-14.2463],[590,-14.1939],[591,-14.1874],[603,-15.1027],[607,-15.2931],[611,-15.3741],[616,-15.3062],[619,-15.1752],[621,-15.0188],[623,-14.758],[625,-14.4143],[627,-14.0008],[629,-13.7749],[633,-12.7816],[635,-12.2428],[676,5.0753],[685,4.9196],[698,4.5657],[716,4.3598],[772,6.4107],[780,8.7579],[787,10.2904],[837,5.0753],[854,3.4479],[887,-11.9645],[893,-10.2853],[895,-9.7609],[897,-9.7688],[903,-9.8645],[929,-10.5519]]
       }
       upperarm_l_fk_ctrl.Rotation.Z {
        type=MovieSceneScriptingFloatChannel
        default=49.1326
        thinned_from=621
        keys [[-2,57.2852],[10,57.3531],[32,60.3347],[52,60.6505],[83,58.1081],[86,57.5104],[95,54.4621],[99,53.6114],[102,53.3474],[105,53.4365],[113,54.9293],[118,56.6758],[126,59.92],[130,60.98],[134,61.4123],[169,59.0672],[218,61.9165],[225,62.6919],[232,62.4398],[241,60.5824],[254,56.1703],[258,55.2465],[262,54.8063],[273,54.9714],[288,55.7389],[296,55.3375],[302,54.4442],[339,57.8098],[356,60.2572],[377,60.6505],[408,58.1081],[411,57.5104],[420,54.4621],[424,53.6114],[427,53.3474],[430,53.4365],[438,54.9293],[443,56.6758],[451,59.92],[455,60.98],[459,61.4123],[494,59.0672],[543,61.9165],[550,62.6919],[557,62.4398],[566,60.5824],[579,56.1703],[583,55.2465],[587,54.8063],[598,54.9714],[613,55.7389],[621,55.3375],[627,54.4442],[629,54.2567],[635,53.0091],[667,-7.6539],[672,-13.0386],[688,-12.1353],[692,-12.3827],[710,-16.9226],[714,-16.5238],[727,-15.3689],[780,-16.0443],[788,-15.8678],[818,-15.3406],[833,-15.796],[837,-15.8811],[848,-10.0196],[872,60.8305],[886,69.5307],[897,56.1761],[903,54.6359],[929,57.2852]]
       }
       upperarm_l_fk_ctrl.Scale.X {type=MovieSceneScriptingFloatChannel, default=1, const [1,642]}
       upperarm_l_fk_ctrl.Scale.Y {type=MovieSceneScriptingFloatChannel, default=1, const [1,642]}
       upperarm_l_fk_ctrl.Scale.Z {type=MovieSceneScriptingFloatChannel, default=1, const [1,642]}
       lowerarm_l_fk_ctrl.Location.X {type=MovieSceneScriptingFloatChannel, default=0, thinned_from=752, keys [[-2,0],[929,0]]}
       lowerarm_l_fk_ctrl.Location.Y {type=MovieSceneScriptingFloatChannel, default=0, thinned_from=752, keys [[-2,0],[929,0]]}
       lowerarm_l_fk_ctrl.Location.Z {type=MovieSceneScriptingFloatChannel, default=0, thinned_from=752, keys [[-2,0],[929,0]]}
       lowerarm_l_fk_ctrl.Rotation.X {
        type=MovieSceneScriptingFloatChannel
        default=0
        thinned_from=752
        keys [[-2,-0.962],[-1,-0.9392],[2,-0.9051],[3,-0.8994],[4,-0.888],[5,-0.8823],[6,-0.8709],[7,-0.8652],[10,-0.8311],[11,-0.8254],[14,-0.7912],[17,-0.768],[21,-0.7489],[25,-0.7351],[26,-0.7351],[27,-0.7287],[29,-0.7287],[30,-0.7213],[33,-0.7217],[34,-0.7104],[36,-0.7104],[37,-0.7048],[40,-0.7048],[41,-0.6991],[42,-0.6991],[43,-0.6935],[45,-0.6935],[46,-0.6878],[47,-0.6866],[48,-0.6823],[49,-0.6823],[52,-0.6696],[53,-0.6696],[60,-0.6398],[61,-0.6312],[64,-0.616],[65,-0.616],[78,-0.5448],[83,-0.5247],[84,-0.5167],[85,-0.5046],[86,-0.4966],[89,-0.4603],[90,-0.4523],[91,-0.4402],[93,-0.4241],[94,-0.4201],[95,-0.4201],[96,-0.4223],[98,-0.4402],[99,-0.4535],[100,-0.4691],[101,-0.4869],[102,-0.507],[104,-0.5516],[107,-0.6251],[108,-0.6518],[111,-0.7243],[113,-0.7684],[114,-0.7867],[115,-0.8032],[116,-0.8161],[117,-0.8252],[118,-0.8326],[119,-0.8344],[120,-0.8326],[121,-0.8271],[122,-0.8234],[124,-0.8124],[131,-0.7862],[132,-0.7862],[136,-0.77],[137,-0.77],[138,-0.7659],[139,-0.7659],[140,-0.7618],[141,-0.7537],[142,-0.7497],[150,-0.7],[151,-0.6909],[155,-0.6665],[156,-0.6635],[157,-0.6574],[159,-0.6513],[160,-0.6513],[161,-0.6483],[162,-0.6483],[163,-0.6452],[164,-0.6452],[172,-0.6209],[173,-0.6148],[174,-0.6118],[175,-0.6118],[176,-0.6037],[178,-0.6037],[179,-0.5955],[184,-0.5955],[185,-0.5874],[195,-0.5874],[196,-0.59],[198,-0.59],[200,-0.5926],[201,-0.5926],[202,-0.5953],[203,-0.5953],[204,-0.5979],[205,-0.5979],[207,-0.6033],[208,-0.6033],[210,-0.6108],[213,-0.6442],[214,-0.6591],[216,-0.6963],[220,-0.7855],[221,-0.8115],[223,-0.849],[226,-0.8928],[229,-0.9147],[231,-0.9147],[233,-0.9001],[236,-0.8563],[238,-0.8125],[239,-0.7933],[241,-0.7446],[242,-0.7243],[245,-0.6513],[247,-0.6108],[249,-0.5783],[250,-0.5661],[251,-0.558],[252,-0.554],[253,-0.554],[254,-0.5418],[255,-0.5406],[256,-0.5382],[259,-0.5382],[261,-0.5431],[263,-0.5518],[264,-0.5553],[265,-0.5627],[266,-0.5676],[270,-0.6069],[278,-0.6927],[281,-0.7201],[283,-0.7338],[285,-0.7433],[288,-0.7517],[291,-0.7517],[292,-0.7487],[293,-0.7487],[295,-0.7426],[296,-0.7426],[298,-0.7365],[302,-0.7365],[304,-0.7334],[309,-0.757],[310,-0.7664],[323,-0.962],[324,-0.9392],[327,-0.9051],[328,-0.8994],[329,-0.888],[330,-0.8823],[331,-0.8709],[332,-0.8652],[335,-0.8311],[336,-0.8254],[339,-0.7912],[342,-0.768],[346,-0.7489],[350,-0.7351],[351,-0.7351],[352,-0.7287],[354,-0.7287],[355,-0.7213],[358,-0.7217],[359,-0.7104],[361,-0.7104],[362,-0.7048],[365,-0.7048],[366,-0.6991],[367,-0.6991],[368,-0.6935],[370,-0.6935],[371,-0.6878],[372,-0.6866],[373,-0.6823],[374,-0.6823],[377,-0.6696],[378,-0.6696],[385,-0.6398],[386,-0.6312],[389,-0.616],[390,-0.616],[403,-0.5448],[408,-0.5247],[409,-0.5167],[410,-0.5046],[411,-0.4966],[414,-0.4603],[415,-0.4523],[416,-0.4402],[418,-0.4241],[419,-0.4201],[420,-0.4201],[421,-0.4223],[423,-0.4402],[424,-0.4535],[425,-0.4691],[426,-0.4869],[427,-0.507],[429,-0.5516],[432,-0.6251],[433,-0.6518],[436,-0.7243],[438,-0.7684],[439,-0.7867],[440,-0.8032],[441,-0.8161],[442,-0.8252],[443,-0.8326],[444,-0.8344],[445,-0.8326],[446,-0.8271],[447,-0.8234],[449,-0.8124],[456,-0.7862],[457,-0.7862],[461,-0.77],[462,-0.77],[463,-0.7659],[464,-0.7659],[465,-0.7618],[466,-0.7537],[467,-0.7497],[475,-0.7],[476,-0.6909],[480,-0.6665],[481,-0.6635],[482,-0.6574],[484,-0.6513],[485,-0.6513],[486,-0.6483],[487,-0.6483],[488,-0.6452],[489,-0.6452],[497,-0.6209],[498,-0.6148],[499,-0.6118],[500,-0.6118],[501,-0.6037],[503,-0.6037],[504,-0.5955],[509,-0.5955],[510,-0.5874],[520,-0.5874],[521,-0.59],[523,-0.59],[525,-0.5926],[526,-0.5926],[527,-0.5953],[528,-0.5953],[529,-0.5979],[530,-0.5979],[532,-0.6033],[533,-0.6033],[535,-0.6108],[538,-0.6442],[539,-0.6591],[541,-0.6963],[545,-0.7855],[546,-0.8115],[548,-0.849],[551,-0.8928],[554,-0.9147],[556,-0.9147],[558,-0.9001],[561,-0.8563],[563,-0.8125],[564,-0.7933],[566,-0.7446],[567,-0.7243],[570,-0.6513],[572,-0.6108],[574,-0.5783],[575,-0.5661],[576,-0.558],[577,-0.554],[578,-0.554],[579,-0.5418],[580,-0.5406],[581,-0.5382],[584,-0.5382],[586,-0.5431],[588,-0.5518],[589,-0.5553],[590,-0.5627],[591,-0.5676],[595,-0.6069],[603,-0.6927],[606,-0.7201],[608,-0.7338],[610,-0.7433],[613,-0.7517],[616,-0.7517],[617,-0.7487],[618,-0.7487],[620,-0.7426],[621,-0.7426],[623,-0.7365],[627,-0.7365],[629,-0.7334],[634,-0.757],[635,-0.7664],[676,2.876],[677,2.8988],[680,2.9329],[688,3.0069],[695,3.07],[703,3.1029],[704,3.1029],[705,3.1093],[707,3.1093],[708,3.1167],[711,3.1163],[712,3.1276],[714,3.1276],[718,3.1332],[738,3.1982],[742,3.222],[759,3.3061],[770,3.1419],[776,3.1022],[788,3.1045],[793,3.081],[797,3.0433],[803,2.9632],[819,2.9105],[837,2.876],[887,-0.7758],[889,-0.7947],[890,-0.8088],[891,-0.8182],[895,-0.8747],[929,-0.962]]
       }
       lowerarm_l_fk_ctrl.Rotation.Y {
        type=MovieSceneScriptingFloatChannel
        default=0
        thinned_from=752
        keys [[-2,-3.329],[635,-3.3296],[676,-1.9335],[759,-1.933],[770,-1.8453],[776,-1.9042],[780,-1.9281],[803,-1.9338],[837,-1.9335],[887,-3.3295],[929,-3.329]]
       }
       lowerarm_l_fk_ctrl.Rotation.Z {
        type=MovieSceneScriptingFloatChannel
        default=0
        thinned_from=664
        keys [[-2,74.9111],[103,74.7676],[127,71.6604],[221,70.4955],[302,75.5335],[310,75.4437],[429,74.6874],[452,71.6604],[546,70.4955],[627,75.5335],[635,75.4437],[657,50.4107],[680,13.2752],[692,17.2578],[696,17.6851],[712,13.6371],[714,13.0733],[728,16.7414],[734,17.7165],[746,16.4503],[759,17.7542],[770,24.8503],[778,28.0531],[785,28.2558],[787,27.8643],[789,27.2768],[803,21.612],[819,17.8414],[837,17.7103],[852,24.8437],[867,60.44],[884,80.6041],[899,75.2986],[929,74.9111]]
       }
       lowerarm_l_fk_ctrl.Scale.X {type=MovieSceneScriptingFloatChannel, default=1, const [1,752]}
       lowerarm_l_fk_ctrl.Scale.Y {type=MovieSceneScriptingFloatChannel, default=1, const [1,752]}
       lowerarm_l_fk_ctrl.Scale.Z {type=MovieSceneScriptingFloatChannel, default=1, const [1,752]}
       hand_l_fk_ctrl.Location.X {type=MovieSceneScriptingFloatChannel, default=0, thinned_from=752, keys [[-2,0],[929,0]]}
       hand_l_fk_ctrl.Location.Y {type=MovieSceneScriptingFloatChannel, default=0, thinned_from=752, keys [[-2,0],[929,0]]}
       hand_l_fk_ctrl.Location.Z {type=MovieSceneScriptingFloatChannel, default=0, thinned_from=752, keys [[-2,0],[929,0]]}
       hand_l_fk_ctrl.Rotation.X {
        type=MovieSceneScriptingFloatChannel
        default=0
        thinned_from=754
        keys [[-2,22.7788],[99,22.261],[217,22.5285],[234,22.7032],[271,22.4259],[310,22.5791],[323,22.7788],[424,22.261],[542,22.5285],[559,22.7032],[596,22.4259],[635,22.5791],[647,-26.4736],[676,-9.1032],[695,-9.3029],[755,-9.5194],[757,-9.5463],[759,-9.6697],[770,-14.5598],[771,-13.8329],[772,-13.1591],[773,-12.5285],[774,-11.9498],[775,-11.422],[776,-10.9476],[777,-10.532],[778,-10.1713],[779,-9.8741],[780,-9.6395],[781,-9.4668],[782,-9.3646],[783,-9.3321],[801,-9.2228],[803,-9.1921],[837,-9.1032],[852,-8.7404],[869,-6.9753],[887,22.5885],[895,22.6899],[929,22.7788]]
       }
       hand_l_fk_ctrl.Rotation.Y {
        type=MovieSceneScriptingFloatChannel
        default=0
        thinned_from=643
        keys [[-2,26.1323],[310,26.2132],[635,26.2132],[648,22.2174],[674,13.8348],[688,7.0708],[699,10.9573],[716,6.6622],[837,7.0397],[851,-11.334],[869,9.8156],[887,26.2094],[895,26.1684],[929,26.1323]]
       }
       hand_l_fk_ctrl.Rotation.Z {
        type=MovieSceneScriptingFloatChannel
        default=0
        thinned_from=753
        keys [[-2,25.4282],[310,25.3394],[635,25.3394],[650,21.0393],[676,7.5951],[759,7.4023],[770,7.74],[777,7.5549],[783,7.4933],[803,7.5556],[837,7.5951],[869,9.3305],[887,25.3436],[895,25.3887],[929,25.4282]]
       }
       hand_l_fk_ctrl.Scale.X {type=MovieSceneScriptingFloatChannel, default=1, const [1,752]}
       hand_l_fk_ctrl.Scale.Y {type=MovieSceneScriptingFloatChannel, default=1, const [1,752]}
       hand_l_fk_ctrl.Scale.Z {type=MovieSceneScriptingFloatChannel, default=1, const [1,752]}
       thumb_01_l_ctrl.Location.X {type=MovieSceneScriptingFloatChannel, default=1.5173, const [1.5173,628]}
       thumb_01_l_ctrl.Location.Y {type=MovieSceneScriptingFloatChannel, default=-0.0715, const [-0.0715,628]}
       thumb_01_l_ctrl.Location.Z {type=MovieSceneScriptingFloatChannel, default=0.0001, const [0.0001,628]}
       thumb_01_l_ctrl.Rotation.X {
        type=MovieSceneScriptingFloatChannel
        default=-23.3104
        thinned_from=629
        keys [[-2,-23.3104],[634,-23.3104],[648,-18.7953],[695,-17.9281],[843,-17.9281],[864,-16.4896],[912,-23.3104]]
       }
       thumb_01_l_ctrl.Rotation.Y {
        type=MovieSceneScriptingFloatChannel
        default=-0.3589
        thinned_from=629
        keys [[-2,-0.3589],[634,-0.3589],[648,30.7495],[695,6.8327],[843,6.8327],[864,20.2529],[912,-0.3589]]
       }
       thumb_01_l_ctrl.Rotation.Z {
        type=MovieSceneScriptingFloatChannel
        default=0.0053
        thinned_from=629
        keys [[-2,0.0053],[634,0.0053],[649,-3.3721],[695,-3.1087],[843,-3.1087],[864,4.8272],[912,0.0053]]
       }
       thumb_01_l_ctrl.Scale.X {type=MovieSceneScriptingFloatChannel, default=1, const [1,628]}
       thumb_01_l_ctrl.Scale.Y {type=MovieSceneScriptingFloatChannel, default=1, const [1,628]}
       thumb_01_l_ctrl.Scale.Z {type=MovieSceneScriptingFloatChannel, default=1, const [1,628]}
       thumb_02_l_ctrl.Location.X {type=MovieSceneScriptingFloatChannel, default=0, thinned_from=629, keys [[-2,0],[912,0]]}
       thumb_02_l_ctrl.Location.Y {type=MovieSceneScriptingFloatChannel, default=0, thinned_from=629, keys [[-2,0],[912,0]]}
       thumb_02_l_ctrl.Location.Z {type=MovieSceneScriptingFloatChannel, default=0, thinned_from=629, keys [[-2,0],[912,0]]}
       thumb_02_l_ctrl.Rotation.X {
        type=MovieSceneScriptingFloatChannel
        default=1.2228
        thinned_from=630
        keys [[-2,1.2228],[634,1.2228],[648,0],[695,1.2228],[843,1.2228],[864,0.5913],[876,0.5581],[912,1.2228]]
       }
       thumb_02_l_ctrl.Rotation.Y {
        type=MovieSceneScriptingFloatChannel
        default=-2.3382
        thinned_from=630
        keys [[-2,-2.3382],[634,-2.3382],[648,0],[695,-2.3382],[843,-2.3382],[864,-2.5668],[876,-2.5788],[912,-2.3382]]
       }
       thumb_02_l_ctrl.Rotation.Z {
        type=MovieSceneScriptingFloatChannel
        default=-13.4617
        thinned_from=631
        keys [[-2,-13.4617],[634,-13.4617],[648,23.2602],[657,26.1313],[695,-13.4617],[843,-13.4617],[864,1.1725],[876,1.9425],[912,-13.4617]]
       }
       thumb_02_l_ctrl.Scale.X {type=MovieSceneScriptingFloatChannel, default=1, const [1,629]}
       thumb_02_l_ctrl.Scale.Y {type=MovieSceneScriptingFloatChannel, default=1, const [1,629]}
       thumb_02_l_ctrl.Scale.Z {type=MovieSceneScriptingFloatChannel, default=1, const [1,629]}
       thumb_03_l_ctrl.Location.X {type=MovieSceneScriptingFloatChannel, default=0, thinned_from=629, keys [[-2,0],[912,0]]}
       thumb_03_l_ctrl.Location.Y {type=MovieSceneScriptingFloatChannel, default=0, thinned_from=629, keys [[-2,0],[912,0]]}
       thumb_03_l_ctrl.Location.Z {type=MovieSceneScriptingFloatChannel, default=0, thinned_from=629, keys [[-2,0],[912,0]]}
       thumb_03_l_ctrl.Rotation.X {
        type=MovieSceneScriptingFloatChannel
        default=2.0016
        thinned_from=630
        keys [[-2,2.0016],[634,2.0016],[648,0],[695,2.0016],[843,2.0016],[864,1.4856],[876,1.4585],[912,2.0016]]
       }
       thumb_03_l_ctrl.Rotation.Y {
        type=MovieSceneScriptingFloatChannel
        default=0.1771
        thinned_from=630
        keys [[-2,0.1771],[634,0.1771],[648,0],[695,0.1771],[843,0.1771],[864,-1.3044],[876,-1.3823],[912,0.1771]]
       }
       thumb_03_l_ctrl.Rotation.Z {
        type=MovieSceneScriptingFloatChannel
        default=-52.3193
        thinned_from=632
        keys [[-2,-52.3193],[634,-52.3193],[647,1.094],[657,11.1008],[694,-52.3193],[843,-52.2978],[864,-6.2347],[876,-3.811],[912,-52.3193]]
       }
       thumb_03_l_ctrl.Scale.X {type=MovieSceneScriptingFloatChannel, default=1, const [1,629]}
       thumb_03_l_ctrl.Scale.Y {type=MovieSceneScriptingFloatChannel, default=1, const [1,629]}
       thumb_03_l_ctrl.Scale.Z {type=MovieSceneScriptingFloatChannel, default=1, const [1,629]}
       index_metacarpal_l_ctrl.Location.X {type=MovieSceneScriptingFloatChannel, default=0, const [0,635]}
       index_metacarpal_l_ctrl.Location.Y {type=MovieSceneScriptingFloatChannel, default=0, thinned_from=635, keys [[-2,0],[929,0]]}
       index_metacarpal_l_ctrl.Location.Z {type=MovieSceneScriptingFloatChannel, default=0, thinned_from=635, keys [[-2,0],[929,0]]}
       index_metacarpal_l_ctrl.Rotation.X {
        type=MovieSceneScriptingFloatChannel
        default=-0.4323
        thinned_from=636
        keys [[-2,-0.4323],[634,-0.4323],[648,0],[864,-0.4323],[929,-0.4323]]
       }
       index_metacarpal_l_ctrl.Rotation.Y {type=MovieSceneScriptingFloatChannel, default=0.5781, thinned_from=636, keys [[-2,0.5781],[634,0.5781],[648,0],[864,0.5781],[929,0.5781]]}
       index_metacarpal_l_ctrl.Rotation.Z {
        type=MovieSceneScriptingFloatChannel
        default=-0.9506
        thinned_from=636
        keys [[-2,-0.9506],[634,-0.9506],[648,0],[864,-0.9506],[929,-0.9506]]
       }
       index_metacarpal_l_ctrl.Scale.X {type=MovieSceneScriptingFloatChannel, default=1, const [1,635]}
       index_metacarpal_l_ctrl.Scale.Y {type=MovieSceneScriptingFloatChannel, default=1, const [1,635]}
       index_metacarpal_l_ctrl.Scale.Z {type=MovieSceneScriptingFloatChannel, default=1, const [1,635]}
       index_01_l_ctrl.Location.X {type=MovieSceneScriptingFloatChannel, default=0, thinned_from=629, keys [[-2,0],[912,0]]}
       index_01_l_ctrl.Location.Y {type=MovieSceneScriptingFloatChannel, default=0, thinned_from=629, keys [[-2,0],[912,0]]}
       index_01_l_ctrl.Location.Z {type=MovieSceneScriptingFloatChannel, default=0, thinned_from=629, keys [[-2,0],[912,0]]}
       index_01_l_ctrl.Rotation.X {
        type=MovieSceneScriptingFloatChannel
        default=3.8515
        thinned_from=631
        keys [[-2,3.8515],[634,3.8515],[645,0.8675],[648,-2.4455],[695,11.3889],[843,11.3889],[864,-1.5646],[876,-2.2462],[912,3.8515]]
       }
       index_01_l_ctrl.Rotation.Y {
        type=MovieSceneScriptingFloatChannel
        default=-11.4472
        thinned_from=632
        keys [[-2,-11.4472],[634,-11.4472],[645,-16.6226],[658,-11.7469],[695,-4.0225],[843,-4.0225],[864,-25.3522],[872,-26.4745],[886,-13.5496],[912,-11.4472]]
       }
       index_01_l_ctrl.Rotation.Z {
        type=MovieSceneScriptingFloatChannel
        default=-4.3782
        thinned_from=631
        keys [[-2,-4.3782],[634,-4.3782],[645,3.2807],[648,14.728],[695,-56.4402],[843,-56.4402],[864,5.7277],[876,8.9988],[912,-4.3782]]
       }
       index_01_l_ctrl.Scale.X {type=MovieSceneScriptingFloatChannel, default=1, const [1,629]}
       index_01_l_ctrl.Scale.Y {type=MovieSceneScriptingFloatChannel, default=1, const [1,629]}
       index_01_l_ctrl.Scale.Z {type=MovieSceneScriptingFloatChannel, default=1, const [1,629]}
       index_02_l_ctrl.Location.X {type=MovieSceneScriptingFloatChannel, default=0.3298, const [0.3298,629]}
       index_02_l_ctrl.Location.Y {type=MovieSceneScriptingFloatChannel, default=-0.1751, const [-0.1751,629]}
       index_02_l_ctrl.Location.Z {type=MovieSceneScriptingFloatChannel, default=-0.0001, const [-0.0001,629]}
       index_02_l_ctrl.Rotation.X {
        type=MovieSceneScriptingFloatChannel
        default=-0.3899
        thinned_from=630
        keys [[-2,-0.3899],[84,-0.3888],[309,-0.3885],[327,-0.3899],[409,-0.3888],[634,-0.3885],[648,0],[695,-0.2126],[843,-0.2126],[864,-0.3774],[876,-0.3878],[912,-0.3899]]
       }
       index_02_l_ctrl.Rotation.Y {
        type=MovieSceneScriptingFloatChannel
        default=-0.035
        thinned_from=630
        keys [[-2,-0.035],[2,-0.035],[3,-0.0342],[8,-0.034],[9,-0.0333],[10,-0.0331],[13,-0.0331],[14,-0.0322],[15,-0.0324],[16,-0.0316],[18,-0.0315],[19,-0.0307],[20,-0.0305],[21,-0.0297],[23,-0.0296],[24,-0.0286],[25,-0.0286],[26,-0.0278],[27,-0.0277],[28,-0.0269],[29,-0.0267],[30,-0.0258],[31,-0.0255],[32,-0.0248],[33,-0.0247],[46,-0.0166],[47,-0.0165],[53,-0.012],[64,-0.0039],[65,-0.0036],[73,0.0028],[82,0.01],[83,0.0102],[91,0.0166],[94,0.0189],[95,0.0194],[96,0.0204],[105,0.0267],[106,0.0272],[120,0.0361],[121,0.0363],[127,0.0398],[128,0.04],[129,0.0407],[130,0.0409],[131,0.0416],[132,0.0417],[133,0.0424],[135,0.0426],[136,0.0433],[137,0.0435],[138,0.0442],[140,0.0444],[141,0.0451],[147,0.0463],[151,0.0468],[152,0.0468],[153,0.0471],[309,0.0471],[323,-0.035],[327,-0.035],[328,-0.0342],[333,-0.034],[334,-0.0333],[335,-0.0331],[338,-0.0331],[339,-0.0322],[340,-0.0324],[341,-0.0316],[343,-0.0315],[344,-0.0307],[345,-0.0305],[346,-0.0297],[348,-0.0296],[349,-0.0286],[350,-0.0286],[351,-0.0278],[352,-0.0277],[353,-0.0269],[354,-0.0267],[355,-0.0258],[356,-0.0255],[357,-0.0248],[358,-0.0247],[371,-0.0166],[372,-0.0165],[378,-0.012],[389,-0.0039],[390,-0.0036],[398,0.0028],[407,0.01],[408,0.0102],[416,0.0166],[419,0.0189],[420,0.0194],[421,0.0204],[430,0.0267],[431,0.0272],[445,0.0361],[446,0.0363],[452,0.0398],[453,0.04],[454,0.0407],[455,0.0409],[456,0.0416],[457,0.0417],[458,0.0424],[460,0.0426],[461,0.0433],[462,0.0435],[463,0.0442],[465,0.0444],[466,0.0451],[472,0.0463],[476,0.0468],[477,0.0468],[478,0.0471],[634,0.0471],[648,0],[695,-0.3282],[843,-0.3282],[864,0.0343],[876,0.0534],[912,-0.035]]
       }
       index_02_l_ctrl.Rotation.Z {
        type=MovieSceneScriptingFloatChannel
        default=9.3395
        thinned_from=630
        keys [[-2,9.3395],[4,9.3395],[5,9.4664],[9,9.4664],[10,9.5934],[13,9.5934],[14,9.7202],[16,9.7202],[17,9.8561],[19,9.8561],[20,9.992],[21,9.992],[22,10.1278],[23,10.1278],[24,10.2636],[26,10.2636],[27,10.3993],[28,10.3993],[30,10.6707],[31,10.6707],[34,10.9459],[35,10.9459],[47,12.0454],[53,12.7078],[54,12.7078],[73,14.8939],[74,14.8939],[91,16.9184],[92,16.9184],[108,18.6504],[109,18.7032],[124,20.049],[125,20.049],[127,20.2234],[128,20.3501],[129,20.3502],[130,20.4769],[131,20.4769],[132,20.6036],[133,20.6036],[134,20.7303],[136,20.7303],[137,20.857],[138,20.857],[139,20.9836],[141,20.9836],[142,21.1102],[143,21.0636],[146,21.2035],[147,21.2035],[148,21.2502],[149,21.2502],[150,21.2968],[152,21.2968],[153,21.3434],[309,21.3434],[323,9.3395],[329,9.3395],[330,9.4664],[334,9.4664],[335,9.5934],[338,9.5934],[339,9.7202],[341,9.7202],[342,9.8561],[344,9.8561],[345,9.992],[346,9.992],[347,10.1278],[348,10.1278],[349,10.2636],[351,10.2636],[352,10.3993],[353,10.3993],[355,10.6707],[356,10.6707],[359,10.9459],[360,10.9459],[372,12.0454],[378,12.7078],[379,12.7078],[398,14.8939],[399,14.8939],[416,16.9184],[417,16.9184],[433,18.6504],[434,18.7032],[449,20.049],[450,20.049],[452,20.2234],[453,20.3501],[454,20.3502],[455,20.4769],[456,20.4769],[457,20.6036],[458,20.6036],[459,20.7303],[461,20.7303],[462,20.857],[463,20.857],[464,20.9836],[466,20.9836],[467,21.1102],[468,21.0636],[471,21.2035],[472,21.2035],[473,21.2502],[474,21.2502],[475,21.2968],[477,21.2968],[478,21.3434],[634,21.3434],[660,10.4991],[695,-42.7068],[843,-42.7068],[864,19.0552],[876,22.3049],[912,9.3395]]
       }
       index_02_l_ctrl.Scale.X {type=MovieSceneScriptingFloatChannel, default=1, const [1,629]}
       index_02_l_ctrl.Scale.Y {type=MovieSceneScriptingFloatChannel, default=1, const [1,629]}
       index_02_l_ctrl.Scale.Z {type=MovieSceneScriptingFloatChannel, default=1, const [1,629]}
       index_03_l_ctrl.Location.X {type=MovieSceneScriptingFloatChannel, default=0.2932, thinned_from=629, keys [[-2,0.2932],[912,0.2932]]}
       index_03_l_ctrl.Location.Y {type=MovieSceneScriptingFloatChannel, default=0.1198, thinned_from=629, keys [[-2,0.1198],[912,0.1198]]}
       index_03_l_ctrl.Location.Z {type=MovieSceneScriptingFloatChannel, default=0.005, thinned_from=629, keys [[-2,0.005],[912,0.005]]}
       index_03_l_ctrl.Rotation.X {
        type=MovieSceneScriptingFloatChannel
        default=-0.4305
        thinned_from=631
        keys [[-2,-0.4305],[634,-0.4305],[645,-0.2588],[648,0],[695,-0.79],[843,-0.79],[864,-0.296],[876,-0.2701],[912,-0.4305]]
       }
       index_03_l_ctrl.Rotation.Y {
        type=MovieSceneScriptingFloatChannel
        default=0.6661
        thinned_from=631
        keys [[-2,0.6661],[634,0.6661],[645,0.7206],[648,0],[695,0.0702],[843,0.0702],[864,0.712],[876,0.7457],[912,0.6661]]
       }
       index_03_l_ctrl.Rotation.Z {
        type=MovieSceneScriptingFloatChannel
        default=0
        thinned_from=631
        keys [[-2,0],[634,0],[645,12.6098],[660,24.5996],[695,-52.0448],[843,-52.0448],[864,9.7167],[876,12.9664],[912,0]]
       }
       index_03_l_ctrl.Scale.X {type=MovieSceneScriptingFloatChannel, default=1, const [1,629]}
       index_03_l_ctrl.Scale.Y {type=MovieSceneScriptingFloatChannel, default=1, const [1,629]}
       index_03_l_ctrl.Scale.Z {type=MovieSceneScriptingFloatChannel, default=1, const [1,629]}
       middle_metacarpal_l_ctrl.Location.X {type=MovieSceneScriptingFloatChannel, default=0, const [0,635]}
       middle_metacarpal_l_ctrl.Location.Y {type=MovieSceneScriptingFloatChannel, default=0, thinned_from=635, keys [[-2,0],[929,0]]}
       middle_metacarpal_l_ctrl.Location.Z {type=MovieSceneScriptingFloatChannel, default=0, thinned_from=635, keys [[-2,0],[929,0]]}
       middle_metacarpal_l_ctrl.Rotation.X {type=MovieSceneScriptingFloatChannel, default=-0.9766, thinned_from=636, keys [[-2,-0.9766],[634,-0.9766],[648,-0.9793],[929,-0.9766]]}
       middle_metacarpal_l_ctrl.Rotation.Y {
        type=MovieSceneScriptingFloatChannel
        default=-4.3812
        thinned_from=636
        keys [[-2,-4.3812],[634,-4.3812],[648,-6.063],[864,-4.3812],[929,-4.3812]]
       }
       middle_metacarpal_l_ctrl.Rotation.Z {
        type=MovieSceneScriptingFloatChannel
        default=0.8157
        thinned_from=636
        keys [[-2,0.8157],[634,0.8157],[648,0.8445],[864,0.8157],[929,0.8157]]
       }
       middle_metacarpal_l_ctrl.Scale.X {type=MovieSceneScriptingFloatChannel, default=1, const [1,635]}
       middle_metacarpal_l_ctrl.Scale.Y {type=MovieSceneScriptingFloatChannel, default=1, const [1,635]}
       middle_metacarpal_l_ctrl.Scale.Z {type=MovieSceneScriptingFloatChannel, default=1, const [1,635]}
       middle_01_l_ctrl.Location.X {type=MovieSceneScriptingFloatChannel, default=0, thinned_from=629, keys [[-2,0],[912,0]]}
       middle_01_l_ctrl.Location.Y {type=MovieSceneScriptingFloatChannel, default=0, thinned_from=629, keys [[-2,0],[912,0]]}
       middle_01_l_ctrl.Location.Z {type=MovieSceneScriptingFloatChannel, default=0, thinned_from=629, keys [[-2,0],[912,0]]}
       middle_01_l_ctrl.Rotation.X {
        type=MovieSceneScriptingFloatChannel
        default=2.3332
        thinned_from=630
        keys [[-2,2.3332],[634,2.3332],[660,3.1508],[695,5.4041],[843,5.4041],[864,2.9775],[876,2.6157],[912,2.3332]]
       }
       middle_01_l_ctrl.Rotation.Y {
        type=MovieSceneScriptingFloatChannel
        default=-5.0358
        thinned_from=631
        keys [[-2,-5.0358],[634,-5.0358],[659,6.0264],[695,-1.2611],[843,-1.2611],[864,-9.1419],[869,-9.779],[884,-4.8064],[912,-5.0358]]
       }
       middle_01_l_ctrl.Rotation.Z {
        type=MovieSceneScriptingFloatChannel
        default=-10.7765
        thinned_from=630
        keys [[-2,-10.7765],[634,-10.7765],[652,2.9437],[695,-62.7803],[843,-62.7803],[864,-1.0424],[876,2.2061],[912,-10.7765]]
       }
       middle_01_l_ctrl.Scale.X {type=MovieSceneScriptingFloatChannel, default=1, const [1,629]}
       middle_01_l_ctrl.Scale.Y {type=MovieSceneScriptingFloatChannel, default=1, const [1,629]}
       middle_01_l_ctrl.Scale.Z {type=MovieSceneScriptingFloatChannel, default=1, const [1,629]}
       middle_02_l_ctrl.Location.X {type=MovieSceneScriptingFloatChannel, default=0.0008, const [0.0008,629]}
       middle_02_l_ctrl.Location.Y {type=MovieSceneScriptingFloatChannel, default=0.0009, const [0.0009,629]}
       middle_02_l_ctrl.Location.Z {type=MovieSceneScriptingFloatChannel, default=0.0476, const [0.0476,629]}
       middle_02_l_ctrl.Rotation.X {
        type=MovieSceneScriptingFloatChannel
        default=-3.8567
        thinned_from=630
        keys [[-2,-3.8567],[309,-3.8415],[634,-3.8415],[648,0],[695,-2.1144],[843,-2.1144],[864,-3.774],[876,-3.8613],[912,-3.8567]]
       }
       middle_02_l_ctrl.Rotation.Y {
        type=MovieSceneScriptingFloatChannel
        default=-0.3333
        thinned_from=630
        keys [[-2,-0.3333],[2,-0.3333],[3,-0.3289],[4,-0.3289],[5,-0.3246],[8,-0.3246],[10,-0.3158],[12,-0.3158],[14,-0.3071],[15,-0.3062],[16,-0.3007],[17,-0.3007],[18,-0.2966],[19,-0.2911],[20,-0.2911],[22,-0.2815],[23,-0.2775],[24,-0.272],[25,-0.272],[26,-0.2624],[27,-0.2624],[30,-0.2481],[31,-0.2391],[32,-0.2387],[39,-0.1955],[40,-0.1924],[46,-0.1554],[47,-0.1471],[48,-0.143],[49,-0.1361],[54,-0.0997],[55,-0.096],[62,-0.045],[63,-0.0357],[66,-0.0119],[67,-0.0075],[70,0.0163],[71,0.0239],[77,0.0715],[78,0.0756],[79,0.0852],[84,0.1246],[85,0.1285],[94,0.1993],[95,0.2041],[109,0.3014],[110,0.3066],[121,0.373],[122,0.3762],[124,0.388],[125,0.391],[126,0.397],[127,0.3985],[128,0.4072],[129,0.4074],[130,0.416],[131,0.4162],[132,0.4248],[133,0.425],[134,0.4337],[135,0.4339],[137,0.4425],[138,0.4427],[139,0.4513],[141,0.4515],[142,0.4601],[143,0.4586],[144,0.4651],[146,0.4653],[148,0.4719],[150,0.4719],[151,0.475],[154,0.475],[155,0.4786],[158,0.4786],[159,0.4771],[309,0.4771],[323,-0.3333],[327,-0.3333],[328,-0.3289],[329,-0.3289],[330,-0.3246],[333,-0.3246],[335,-0.3158],[337,-0.3158],[339,-0.3071],[340,-0.3062],[341,-0.3007],[342,-0.3007],[343,-0.2966],[344,-0.2911],[345,-0.2911],[347,-0.2815],[348,-0.2775],[349,-0.272],[350,-0.272],[351,-0.2624],[352,-0.2624],[355,-0.2481],[356,-0.2391],[357,-0.2387],[364,-0.1955],[365,-0.1924],[371,-0.1554],[372,-0.1471],[373,-0.143],[374,-0.1361],[379,-0.0997],[380,-0.096],[387,-0.045],[388,-0.0357],[391,-0.0119],[392,-0.0075],[395,0.0163],[396,0.0239],[402,0.0715],[403,0.0756],[404,0.0852],[409,0.1246],[410,0.1285],[419,0.1993],[420,0.2041],[434,0.3014],[435,0.3066],[446,0.373],[447,0.3762],[449,0.388],[450,0.391],[451,0.397],[452,0.3985],[453,0.4072],[454,0.4074],[455,0.416],[456,0.4162],[457,0.4248],[458,0.425],[459,0.4337],[460,0.4339],[462,0.4425],[463,0.4427],[464,0.4513],[466,0.4515],[467,0.4601],[468,0.4586],[469,0.4651],[471,0.4653],[473,0.4719],[475,0.4719],[476,0.475],[479,0.475],[480,0.4786],[483,0.4786],[484,0.4771],[634,0.4771],[648,0],[695,-3.2424],[843,-3.2424],[864,0.0999],[876,0.2758],[912,-0.3333]]
       }
       middle_02_l_ctrl.Rotation.Z {
        type=MovieSceneScriptingFloatChannel
        default=0.6939
        thinned_from=630
        keys [[-2,0.6939],[4,0.6937],[5,0.8213],[9,0.8212],[10,0.9487],[13,0.9485],[14,1.076],[15,1.0281],[16,1.1714],[18,1.1713],[19,1.3145],[21,1.3144],[22,1.4576],[23,1.4575],[24,1.6006],[25,1.6006],[26,1.7435],[28,1.7434],[29,1.8864],[30,1.8862],[31,2.0736],[32,2.0736],[39,2.7169],[40,2.7168],[46,3.2672],[47,3.4063],[54,4.1574],[55,4.1573],[62,4.9067],[63,5.0608],[77,6.6858],[78,6.6856],[79,6.8484],[94,8.5808],[95,8.6274],[110,10.1613],[111,10.2122],[124,11.38],[125,11.3799],[126,11.4696],[127,11.513],[128,11.6455],[129,11.6455],[130,11.7779],[131,11.7779],[132,11.9103],[133,11.9103],[134,12.0427],[136,12.0426],[137,12.175],[138,12.175],[139,12.3073],[141,12.3074],[142,12.4396],[143,12.3933],[144,12.5013],[147,12.5012],[148,12.6093],[154,12.6092],[155,12.7172],[158,12.7172],[159,12.6709],[309,12.6709],[323,0.6939],[329,0.6937],[330,0.8213],[334,0.8212],[335,0.9487],[338,0.9485],[339,1.076],[340,1.0281],[341,1.1714],[343,1.1713],[344,1.3145],[346,1.3144],[347,1.4576],[348,1.4575],[349,1.6006],[350,1.6006],[351,1.7435],[353,1.7434],[354,1.8864],[355,1.8862],[356,2.0736],[357,2.0736],[364,2.7169],[365,2.7168],[371,3.2672],[372,3.4063],[379,4.1574],[380,4.1573],[387,4.9067],[388,5.0608],[402,6.6858],[403,6.6856],[404,6.8484],[419,8.5808],[420,8.6274],[435,10.1613],[436,10.2122],[449,11.38],[450,11.3799],[451,11.4696],[452,11.513],[453,11.6455],[454,11.6455],[455,11.7779],[456,11.7779],[457,11.9103],[458,11.9103],[459,12.0427],[461,12.0426],[462,12.175],[463,12.175],[464,12.3073],[466,12.3074],[467,12.4396],[468,12.3933],[469,12.5013],[472,12.5012],[473,12.6093],[479,12.6092],[480,12.7172],[483,12.7172],[484,12.6709],[634,12.6709],[648,12.6141],[695,-51.3043],[843,-51.3043],[864,6.6551],[876,9.7048],[912,0.6939]]
       }
       middle_02_l_ctrl.Scale.X {type=MovieSceneScriptingFloatChannel, default=1, const [1,629]}
       middle_02_l_ctrl.Scale.Y {type=MovieSceneScriptingFloatChannel, default=1, const [1,629]}
       middle_02_l_ctrl.Scale.Z {type=MovieSceneScriptingFloatChannel, default=1, const [1,629]}
       middle_03_l_ctrl.Location.X {type=MovieSceneScriptingFloatChannel, default=0, thinned_from=629, keys [[-2,0],[912,0]]}
       middle_03_l_ctrl.Location.Y {type=MovieSceneScriptingFloatChannel, default=0, thinned_from=629, keys [[-2,0],[912,0]]}
       middle_03_l_ctrl.Location.Z {type=MovieSceneScriptingFloatChannel, default=0, thinned_from=629, keys [[-2,0],[912,0]]}
       middle_03_l_ctrl.Rotation.X {
        type=MovieSceneScriptingFloatChannel
        default=0.744
        thinned_from=630
        keys [[-2,0.744],[634,0.744],[648,0],[695,-1.3022],[843,-1.3022],[864,0.9658],[876,1.0851],[912,0.744]]
       }
       middle_03_l_ctrl.Rotation.Y {
        type=MovieSceneScriptingFloatChannel
        default=2.2308
        thinned_from=630
        keys [[-2,2.2308],[634,2.2308],[648,0],[695,1.9582],[843,1.9582],[864,1.9966],[876,2.0863],[912,2.2308]]
       }
       middle_03_l_ctrl.Rotation.Z {
        type=MovieSceneScriptingFloatChannel
        default=0
        thinned_from=630
        keys [[-2,0],[634,0],[648,13.7873],[695,-52.0836],[843,-52.0836],[864,5.9815],[876,9.0367],[912,0]]
       }
       middle_03_l_ctrl.Scale.X {type=MovieSceneScriptingFloatChannel, default=1, const [1,629]}
       middle_03_l_ctrl.Scale.Y {type=MovieSceneScriptingFloatChannel, default=1, const [1,629]}
       middle_03_l_ctrl.Scale.Z {type=MovieSceneScriptingFloatChannel, default=1, const [1,629]}
       ring_metacarpal_l_ctrl.Location.X {type=MovieSceneScriptingFloatChannel, default=0, thinned_from=635, keys [[-2,0],[929,0]]}
       ring_metacarpal_l_ctrl.Location.Y {type=MovieSceneScriptingFloatChannel, default=0, thinned_from=635, keys [[-2,0],[929,0]]}
       ring_metacarpal_l_ctrl.Location.Z {type=MovieSceneScriptingFloatChannel, default=0, thinned_from=635, keys [[-2,0],[929,0]]}
       ring_metacarpal_l_ctrl.Rotation.X {type=MovieSceneScriptingFloatChannel, default=-2.6065, thinned_from=636, keys [[-2,-2.6065],[634,-2.6065],[648,-2.6222],[929,-2.6065]]}
       ring_metacarpal_l_ctrl.Rotation.Y {
        type=MovieSceneScriptingFloatChannel
        default=-5.9585
        thinned_from=636
        keys [[-2,-5.9585],[634,-5.9585],[648,-8.6501],[864,-5.9585],[929,-5.9585]]
       }
       ring_metacarpal_l_ctrl.Rotation.Z {
        type=MovieSceneScriptingFloatChannel
        default=9.7527
        thinned_from=636
        keys [[-2,9.7527],[634,9.7527],[648,9.8766],[864,9.7527],[929,9.7527]]
       }
       ring_metacarpal_l_ctrl.Scale.X {type=MovieSceneScriptingFloatChannel, default=1, const [1,635]}
       ring_metacarpal_l_ctrl.Scale.Y {type=MovieSceneScriptingFloatChannel, default=1, const [1,635]}
       ring_metacarpal_l_ctrl.Scale.Z {type=MovieSceneScriptingFloatChannel, default=1, const [1,635]}
       ring_01_l_ctrl.Location.X {type=MovieSceneScriptingFloatChannel, default=0, thinned_from=629, keys [[-2,0],[912,0]]}
       ring_01_l_ctrl.Location.Y {type=MovieSceneScriptingFloatChannel, default=0, thinned_from=629, keys [[-2,0],[912,0]]}
       ring_01_l_ctrl.Location.Z {type=MovieSceneScriptingFloatChannel, default=0, thinned_from=629, keys [[-2,0],[912,0]]}
       ring_01_l_ctrl.Rotation.X {
        type=MovieSceneScriptingFloatChannel
        default=1.5312
        thinned_from=630
        keys [[-2,1.5312],[0,1.5312],[1,1.5416],[2,1.5416],[3,1.521],[4,1.5205],[5,1.531],[6,1.531],[7,1.5208],[8,1.5204],[9,1.5102],[10,1.5098],[11,1.4997],[12,1.4792],[13,1.4892],[14,1.4788],[15,1.4832],[16,1.4578],[17,1.4444],[18,1.4557],[22,1.4016],[23,1.4013],[26,1.3611],[27,1.3594],[30,1.3198],[31,1.291],[32,1.2889],[35,1.2337],[36,1.2318],[37,1.2064],[40,1.1523],[41,1.1417],[46,1.0543],[47,1.0074],[48,1.023],[59,0.7736],[62,0.7087],[63,0.6832],[64,0.6762],[66,0.6209],[67,0.5998],[73,0.4409],[74,0.4138],[75,0.4079],[76,0.3762],[79,0.2999],[80,0.2713],[81,0.2465],[83,0.1899],[84,0.1656],[86,0.1098],[88,0.0546],[89,0.0272],[90,-0.0054],[91,-0.0326],[92,-0.0633],[93,-0.0903],[94,-0.1207],[95,-0.1466],[96,-0.1736],[99,-0.2565],[100,-0.2847],[102,-0.3424],[103,-0.3718],[104,-0.4031],[107,-0.4937],[108,-0.525],[109,-0.5584],[110,-0.5895],[111,-0.6191],[112,-0.645],[113,-0.6776],[114,-0.7033],[115,-0.7397],[116,-0.7653],[122,-0.9583],[123,-0.9609],[124,-0.997],[125,-1.0357],[126,-1.0674],[127,-1.0901],[130,-1.1849],[131,-1.2019],[133,-1.2649],[134,-1.2995],[138,-1.4251],[139,-1.4596],[140,-1.471],[141,-1.5023],[142,-1.5514],[143,-1.5667],[144,-1.5982],[145,-1.6222],[149,-1.7432],[151,-1.7918],[152,-1.8246],[156,-1.9348],[157,-2.8389],[309,-2.8393],[323,1.5312],[325,1.5312],[326,1.5416],[327,1.5416],[328,1.521],[329,1.5205],[330,1.531],[331,1.531],[332,1.5208],[333,1.5204],[334,1.5102],[335,1.5098],[336,1.4997],[337,1.4792],[338,1.4892],[339,1.4788],[340,1.4832],[341,1.4578],[342,1.4444],[343,1.4557],[347,1.4016],[348,1.4013],[351,1.3611],[352,1.3594],[355,1.3198],[356,1.291],[357,1.2889],[360,1.2337],[361,1.2318],[362,1.2064],[365,1.1523],[366,1.1417],[371,1.0543],[372,1.0074],[373,1.023],[384,0.7736],[387,0.7087],[388,0.6832],[389,0.6762],[391,0.6209],[392,0.5998],[398,0.4409],[399,0.4138],[400,0.4079],[401,0.3762],[404,0.2999],[405,0.2713],[406,0.2465],[408,0.1899],[409,0.1656],[411,0.1098],[413,0.0546],[414,0.0272],[415,-0.0054],[416,-0.0326],[417,-0.0633],[418,-0.0903],[419,-0.1207],[420,-0.1466],[421,-0.1736],[424,-0.2565],[425,-0.2847],[427,-0.3424],[428,-0.3718],[429,-0.4031],[432,-0.4937],[433,-0.525],[434,-0.5584],[435,-0.5895],[436,-0.6191],[437,-0.645],[438,-0.6776],[439,-0.7033],[440,-0.7397],[441,-0.7653],[447,-0.9583],[448,-0.9609],[449,-0.997],[450,-1.0357],[451,-1.0674],[452,-1.0901],[455,-1.1849],[456,-1.2019],[458,-1.2649],[459,-1.2995],[463,-1.4251],[464,-1.4596],[465,-1.471],[466,-1.5023],[467,-1.5514],[468,-1.5667],[469,-1.5982],[470,-1.6222],[474,-1.7432],[476,-1.7918],[477,-1.8246],[481,-1.9348],[482,-2.8389],[634,-2.8393],[660,-2.9641],[695,6.2932],[843,6.2932],[864,1.9172],[876,1.6164],[912,1.5312]]
       }
       ring_01_l_ctrl.Rotation.Y {
        type=MovieSceneScriptingFloatChannel
        default=-6.7802
        thinned_from=630
        keys [[-2,-6.7802],[2,-6.7764],[3,-6.7204],[6,-6.7182],[7,-6.6584],[8,-6.66],[9,-6.6002],[10,-6.6018],[12,-6.486],[13,-6.4838],[14,-6.4256],[15,-6.4018],[17,-6.2618],[18,-6.2591],[22,-5.9721],[23,-5.9736],[30,-5.4696],[31,-5.3565],[32,-5.3065],[35,-5.0206],[36,-4.9715],[46,-4.0211],[47,-3.8613],[48,-3.8585],[62,-2.2561],[63,-2.1327],[64,-2.0707],[74,-0.8413],[75,-0.7792],[76,-0.6585],[78,-0.4131],[79,-0.2921],[80,-0.174],[81,-0.0547],[83,0.1816],[84,0.301],[94,1.482],[95,1.5853],[110,3.1951],[122,4.3123],[123,4.3088],[126,4.5862],[139,5.3887],[140,5.3835],[142,5.5025],[143,5.5184],[146,5.6166],[149,5.6893],[154,5.7531],[156,5.7597],[157,5.3727],[309,5.3728],[323,-6.7802],[327,-6.7764],[328,-6.7204],[331,-6.7182],[332,-6.6584],[333,-6.66],[334,-6.6002],[335,-6.6018],[337,-6.486],[338,-6.4838],[339,-6.4256],[340,-6.4018],[342,-6.2618],[343,-6.2591],[347,-5.9721],[348,-5.9736],[355,-5.4696],[356,-5.3565],[357,-5.3065],[360,-5.0206],[361,-4.9715],[371,-4.0211],[372,-3.8613],[373,-3.8585],[387,-2.2561],[388,-2.1327],[389,-2.0707],[399,-0.8413],[400,-0.7792],[401,-0.6585],[403,-0.4131],[404,-0.2921],[405,-0.174],[406,-0.0547],[408,0.1816],[409,0.301],[419,1.482],[420,1.5853],[435,3.1951],[447,4.3123],[448,4.3088],[451,4.5862],[464,5.3887],[465,5.3835],[467,5.5025],[468,5.5184],[471,5.6166],[474,5.6893],[479,5.7531],[481,5.7597],[482,5.3727],[634,5.3728],[659,11.8544],[695,-2.9522],[843,-2.9522],[864,0.4325],[876,0.6106],[912,-6.7802]]
       }
       ring_01_l_ctrl.Rotation.Z {
        type=MovieSceneScriptingFloatChannel
        default=-18.382
        thinned_from=630
        keys [[-2,-18.382],[3,-18.3761],[4,-18.3396],[7,-18.3345],[14,-18.1653],[15,-18.1638],[21,-17.9585],[30,-17.633],[31,-17.5668],[45,-16.8719],[46,-16.7937],[47,-16.7688],[53,-16.3602],[54,-16.3511],[64,-15.6672],[65,-15.6268],[73,-15.0563],[89,-13.9732],[90,-13.9356],[103,-13.0775],[104,-13.0429],[108,-12.7849],[109,-12.7507],[114,-12.4385],[115,-12.4355],[123,-11.9952],[124,-11.9929],[127,-11.829],[133,-11.6065],[134,-11.6056],[138,-11.4586],[139,-11.4579],[141,-11.3857],[142,-11.3855],[143,-11.3502],[144,-11.3501],[145,-11.2779],[149,-11.2787],[150,-11.2068],[156,-11.2099],[157,-20.5083],[309,-20.5083],[323,-18.382],[328,-18.3761],[329,-18.3396],[332,-18.3345],[339,-18.1653],[340,-18.1638],[346,-17.9585],[355,-17.633],[356,-17.5668],[370,-16.8719],[371,-16.7937],[372,-16.7688],[378,-16.3602],[379,-16.3511],[389,-15.6672],[390,-15.6268],[398,-15.0563],[414,-13.9732],[415,-13.9356],[428,-13.0775],[429,-13.0429],[433,-12.7849],[434,-12.7507],[439,-12.4385],[440,-12.4355],[448,-11.9952],[449,-11.9929],[452,-11.829],[458,-11.6065],[459,-11.6056],[463,-11.4586],[464,-11.4579],[466,-11.3857],[467,-11.3855],[468,-11.3502],[469,-11.3501],[470,-11.2779],[474,-11.2787],[475,-11.2068],[481,-11.2099],[482,-20.5083],[634,-20.5083],[648,-12.0691],[695,-70.5006],[843,-70.5006],[864,-13.5231],[876,-10.5251],[912,-18.382]]
       }
       ring_01_l_ctrl.Scale.X {type=MovieSceneScriptingFloatChannel, default=1, const [1,629]}
       ring_01_l_ctrl.Scale.Y {type=MovieSceneScriptingFloatChannel, default=1, const [1,629]}
       ring_01_l_ctrl.Scale.Z {type=MovieSceneScriptingFloatChannel, default=1, const [1,629]}
       ring_02_l_ctrl.Location.X {type=MovieSceneScriptingFloatChannel, default=0, thinned_from=629, keys [[-2,0],[912,0]]}
       ring_02_l_ctrl.Location.Y {type=MovieSceneScriptingFloatChannel, default=0, thinned_from=629, keys [[-2,0],[912,0]]}
       ring_02_l_ctrl.Location.Z {type=MovieSceneScriptingFloatChannel, default=0, thinned_from=629, keys [[-2,0],[912,0]]}
       ring_02_l_ctrl.Rotation.X {
        type=MovieSceneScriptingFloatChannel
        default=-2.4658
        thinned_from=630
        keys [[-2,-2.4658],[8,-2.4568],[12,-2.4434],[27,-2.3814],[44,-2.2695],[47,-2.2429],[48,-2.2389],[78,-1.9734],[79,-1.9621],[91,-1.8527],[119,-1.628],[136,-1.5342],[142,-1.5119],[151,-1.4925],[152,-1.4925],[153,-1.489],[309,-1.489],[323,-2.4658],[333,-2.4568],[337,-2.4434],[352,-2.3814],[369,-2.2695],[372,-2.2429],[373,-2.2389],[403,-1.9734],[404,-1.9621],[416,-1.8527],[444,-1.628],[461,-1.5342],[467,-1.5119],[476,-1.4925],[477,-1.4925],[478,-1.489],[634,-1.489],[648,0],[695,-3.2214],[843,-3.2214],[864,-2.2091],[876,-2.1558],[912,-2.4658]]
       }
       ring_02_l_ctrl.Rotation.Y {
        type=MovieSceneScriptingFloatChannel
        default=2.1623
        thinned_from=630
        keys [[-2,2.1623],[8,2.1691],[32,2.2505],[46,2.3272],[125,2.8433],[309,2.922],[323,2.1623],[333,2.1691],[357,2.2505],[371,2.3272],[450,2.8433],[634,2.922],[648,0],[695,-0.6129],[843,-0.6129],[864,2.3173],[876,2.4715],[912,2.1623]]
       }
       ring_02_l_ctrl.Rotation.Z {
        type=MovieSceneScriptingFloatChannel
        default=-7.2883
        thinned_from=630
        keys [[-2,-7.2883],[2,-7.2883],[3,-7.188],[5,-7.188],[6,-7.0878],[8,-7.0877],[10,-6.8873],[11,-6.8873],[12,-6.7871],[13,-6.7871],[14,-6.687],[15,-6.5519],[16,-6.5518],[18,-6.3068],[19,-6.3067],[30,-4.9619],[31,-4.8455],[32,-4.764],[36,-4.1128],[37,-4.0315],[46,-2.5709],[47,-2.3113],[48,-2.2137],[55,-0.8493],[56,-0.752],[60,0.0253],[62,0.4134],[63,0.659],[66,1.2798],[67,1.4349],[78,3.7026],[79,3.9612],[80,4.0656],[94,6.9794],[95,7.1302],[97,7.5061],[98,7.7409],[104,8.8666],[105,9.0072],[109,9.7562],[110,9.8966],[111,10.1156],[120,11.531],[121,11.6096],[125,12.2375],[126,12.316],[127,12.5008],[136,13.506],[137,13.5061],[140,13.8408],[141,13.8408],[142,13.9524],[143,13.9802],[146,14.2312],[147,14.2312],[148,14.3148],[149,14.3148],[150,14.3984],[152,14.3984],[153,14.482],[309,14.482],[323,-7.2883],[327,-7.2883],[328,-7.188],[330,-7.188],[331,-7.0878],[333,-7.0877],[335,-6.8873],[336,-6.8873],[337,-6.7871],[338,-6.7871],[339,-6.687],[340,-6.5519],[341,-6.5518],[343,-6.3068],[344,-6.3067],[355,-4.9619],[356,-4.8455],[357,-4.764],[361,-4.1128],[362,-4.0315],[371,-2.5709],[372,-2.3113],[373,-2.2137],[380,-0.8493],[381,-0.752],[385,0.0253],[387,0.4134],[388,0.659],[391,1.2798],[392,1.4349],[403,3.7026],[404,3.9612],[405,4.0656],[419,6.9794],[420,7.1302],[422,7.5061],[423,7.7409],[429,8.8666],[430,9.0072],[434,9.7562],[435,9.8966],[436,10.1156],[445,11.531],[446,11.6096],[450,12.2375],[451,12.316],[452,12.5008],[461,13.506],[462,13.5061],[465,13.8408],[466,13.8408],[467,13.9524],[468,13.9802],[471,14.2312],[472,14.2312],[473,14.3148],[474,14.3148],[475,14.3984],[477,14.3984],[478,14.482],[634,14.482],[648,13.9318],[695,-59.2714],[843,-59.2714],[864,-2.6129],[876,0.3684],[912,-7.2883]]
       }
       ring_02_l_ctrl.Scale.X {type=MovieSceneScriptingFloatChannel, default=1, const [1,629]}
       ring_02_l_ctrl.Scale.Y {type=MovieSceneScriptingFloatChannel, default=1, const [1,629]}
       ring_02_l_ctrl.Scale.Z {type=MovieSceneScriptingFloatChannel, default=1, const [1,629]}
       ring_03_l_ctrl.Location.X {type=MovieSceneScriptingFloatChannel, default=0, const [0,629]}
       ring_03_l_ctrl.Location.Y {type=MovieSceneScriptingFloatChannel, default=0, const [0,629]}
       ring_03_l_ctrl.Location.Z {type=MovieSceneScriptingFloatChannel, default=0, const [0,629]}
       ring_03_l_ctrl.Rotation.X {type=MovieSceneScriptingFloatChannel, default=0, thinned_from=629, keys [[-2,0],[912,0]]}
       ring_03_l_ctrl.Rotation.Y {type=MovieSceneScriptingFloatChannel, default=0, thinned_from=629, keys [[-2,0],[912,0]]}
       ring_03_l_ctrl.Rotation.Z {
        type=MovieSceneScriptingFloatChannel
        default=0
        thinned_from=630
        keys [[-2,0],[634,0],[648,9.8558],[695,-52.0468],[843,-52.0468],[864,4.6722],[876,7.6566],[912,0]]
       }
       ring_03_l_ctrl.Scale.X {type=MovieSceneScriptingFloatChannel, default=1, const [1,629]}
       ring_03_l_ctrl.Scale.Y {type=MovieSceneScriptingFloatChannel, default=1, const [1,629]}
       ring_03_l_ctrl.Scale.Z {type=MovieSceneScriptingFloatChannel, default=1, const [1,629]}
       pinky_metacarpal_l_ctrl.Location.X {type=MovieSceneScriptingFloatChannel, default=0, thinned_from=635, keys [[-2,0],[929,0]]}
       pinky_metacarpal_l_ctrl.Location.Y {type=MovieSceneScriptingFloatChannel, default=0, thinned_from=635, keys [[-2,0],[929,0]]}
       pinky_metacarpal_l_ctrl.Location.Z {type=MovieSceneScriptingFloatChannel, default=0, thinned_from=635, keys [[-2,0],[929,0]]}
       pinky_metacarpal_l_ctrl.Rotation.X {type=MovieSceneScriptingFloatChannel, default=-7.089, thinned_from=636, keys [[-2,-7.089],[929,-7.089]]}
       pinky_metacarpal_l_ctrl.Rotation.Y {
        type=MovieSceneScriptingFloatChannel
        default=-6.2626
        thinned_from=636
        keys [[-2,-6.2626],[634,-6.2626],[648,-6.5247],[864,-6.2626],[929,-6.2626]]
       }
       pinky_metacarpal_l_ctrl.Rotation.Z {
        type=MovieSceneScriptingFloatChannel
        default=2.9135
        thinned_from=636
        keys [[-2,2.9135],[634,2.9135],[648,2.9463],[864,2.9135],[929,2.9135]]
       }
       pinky_metacarpal_l_ctrl.Scale.X {type=MovieSceneScriptingFloatChannel, default=1, const [1,635]}
       pinky_metacarpal_l_ctrl.Scale.Y {type=MovieSceneScriptingFloatChannel, default=1, const [1,635]}
       pinky_metacarpal_l_ctrl.Scale.Z {type=MovieSceneScriptingFloatChannel, default=1, const [1,635]}
       pinky_01_l_ctrl.Location.X {type=MovieSceneScriptingFloatChannel, default=0, thinned_from=629, keys [[-2,0],[912,0]]}
       pinky_01_l_ctrl.Location.Y {type=MovieSceneScriptingFloatChannel, default=0, thinned_from=629, keys [[-2,0],[912,0]]}
       pinky_01_l_ctrl.Location.Z {type=MovieSceneScriptingFloatChannel, default=0, thinned_from=629, keys [[-2,0],[912,0]]}
       pinky_01_l_ctrl.Rotation.X {
        type=MovieSceneScriptingFloatChannel
        default=-3.7473
        thinned_from=630
        keys [[-2,-3.7473],[634,-3.7473],[648,1.3483],[695,8.2015],[843,8.2015],[864,-3.0406],[876,-3.6843],[912,-3.7473]]
       }
       pinky_01_l_ctrl.Rotation.Y {
        type=MovieSceneScriptingFloatChannel
        default=-13.136
        thinned_from=631
        keys [[-2,-13.136],[634,-13.136],[659,9.7733],[666,3.9836],[695,-10.9503],[843,-10.9503],[864,6.9745],[876,7.9176],[912,-13.136]]
       }
       pinky_01_l_ctrl.Rotation.Z {
        type=MovieSceneScriptingFloatChannel
        default=-0.1797
        thinned_from=630
        keys [[-2,-0.1797],[634,-0.1797],[648,5.0016],[695,-53.4456],[843,-53.4456],[864,-5.2071],[876,-1.5407],[912,-0.1797]]
       }
       pinky_01_l_ctrl.Scale.X {type=MovieSceneScriptingFloatChannel, default=1, const [1,629]}
       pinky_01_l_ctrl.Scale.Y {type=MovieSceneScriptingFloatChannel, default=1, const [1,629]}
       pinky_01_l_ctrl.Scale.Z {type=MovieSceneScriptingFloatChannel, default=1, const [1,629]}
       pinky_02_l_ctrl.Location.X {type=MovieSceneScriptingFloatChannel, default=0, thinned_from=629, keys [[-2,0],[912,0]]}
       pinky_02_l_ctrl.Location.Y {type=MovieSceneScriptingFloatChannel, default=0, thinned_from=629, keys [[-2,0],[912,0]]}
       pinky_02_l_ctrl.Location.Z {type=MovieSceneScriptingFloatChannel, default=0, thinned_from=629, keys [[-2,0],[912,0]]}
       pinky_02_l_ctrl.Rotation.X {
        type=MovieSceneScriptingFloatChannel
        default=0.9973
        thinned_from=630
        keys [[-2,0.9973],[634,0.9973],[648,0],[695,-2.1771],[843,-2.1771],[864,0.8386],[876,0.9973],[912,0.9973]]
       }
       pinky_02_l_ctrl.Rotation.Y {
        type=MovieSceneScriptingFloatChannel
        default=3.5352
        thinned_from=630
        keys [[-2,3.5352],[634,3.5352],[648,0],[695,2.959],[843,2.959],[864,3.5064],[876,3.5352],[912,3.5352]]
       }
       pinky_02_l_ctrl.Rotation.Z {
        type=MovieSceneScriptingFloatChannel
        default=16.3636
        thinned_from=629
        keys [[-2,16.3636],[634,16.3636],[695,-35.7703],[843,-35.7703],[864,13.7575],[876,16.3636],[912,16.3636]]
       }
       pinky_02_l_ctrl.Scale.X {type=MovieSceneScriptingFloatChannel, default=1, const [1,629]}
       pinky_02_l_ctrl.Scale.Y {type=MovieSceneScriptingFloatChannel, default=1, const [1,629]}
       pinky_02_l_ctrl.Scale.Z {type=MovieSceneScriptingFloatChannel, default=1, const [1,629]}
       pinky_03_l_ctrl.Location.X {type=MovieSceneScriptingFloatChannel, default=0, thinned_from=629, keys [[-2,0],[912,0]]}
       pinky_03_l_ctrl.Location.Y {type=MovieSceneScriptingFloatChannel, default=0, thinned_from=629, keys [[-2,0],[912,0]]}
       pinky_03_l_ctrl.Location.Z {type=MovieSceneScriptingFloatChannel, default=0, thinned_from=629, keys [[-2,0],[912,0]]}
       pinky_03_l_ctrl.Rotation.X {
        type=MovieSceneScriptingFloatChannel
        default=-0.1956
        thinned_from=630
        keys [[-2,-0.1956],[634,-0.1956],[648,0],[695,-0.9949],[843,-0.9949],[864,-0.2356],[876,-0.1956],[912,-0.1956]]
       }
       pinky_03_l_ctrl.Rotation.Y {
        type=MovieSceneScriptingFloatChannel
        default=1.1092
        thinned_from=630
        keys [[-2,1.1092],[634,1.1092],[648,0],[695,0.5279],[843,0.5279],[864,1.0801],[876,1.1092],[912,1.1092]]
       }
       pinky_03_l_ctrl.Rotation.Z {
        type=MovieSceneScriptingFloatChannel
        default=0
        thinned_from=629
        keys [[-2,0],[634,0],[695,-52.0496],[843,-52.0496],[864,-2.6018],[876,0],[912,0]]
       }
       pinky_03_l_ctrl.Scale.X {type=MovieSceneScriptingFloatChannel, default=1, const [1,629]}
       pinky_03_l_ctrl.Scale.Y {type=MovieSceneScriptingFloatChannel, default=1, const [1,629]}
       pinky_03_l_ctrl.Scale.Z {type=MovieSceneScriptingFloatChannel, default=1, const [1,629]}
       hand_l_ik_ctrl.Location.X {
        type=MovieSceneScriptingFloatChannel
        default=0
        thinned_from=751
        keys [[-2,-60.3991],[-1,-60.8728],[3,-61.1541],[9,-60.5818],[25,-57.9081],[30,-57.6101],[62,-59.6689],[71,-59.2279],[94,-55.68],[100,-55.4546],[107,-56.0921],[117,-58.3358],[131,-62.3553],[138,-63.3048],[220,-58.4689],[231,-57.1164],[238,-57.0532],[279,-62.0689],[293,-62.3108],[304,-60.9725],[309,-60.0102],[323,-60.3991],[324,-60.8728],[328,-61.1541],[334,-60.5818],[350,-57.9081],[355,-57.6101],[387,-59.6689],[396,-59.2279],[419,-55.68],[425,-55.4546],[432,-56.0921],[442,-58.3358],[456,-62.3553],[463,-63.3048],[545,-58.4689],[556,-57.1164],[563,-57.0532],[604,-62.0689],[618,-62.3108],[629,-60.9725],[634,-60.0102],[676,-60.3991],[677,-60.8728],[681,-61.1541],[687,-60.5818],[703,-57.9081],[708,-57.6101],[740,-59.6689],[749,-59.2279],[759,-57.6727],[770,-62.3345],[780,-62.1696],[803,-58.2856],[837,-60.3991],[887,-59.8083],[895,-58.2856],[929,-60.3991]]
       }
       hand_l_ik_ctrl.Location.Y {
        type=MovieSceneScriptingFloatChannel
        default=0
        thinned_from=751
        keys [[-2,-1.8073],[-1,-2.2119],[0,-2.3238],[1,-2.425],[2,-2.5117],[3,-2.5765],[4,-2.6252],[5,-2.6553],[6,-2.6721],[7,-2.6737],[8,-2.6665],[10,-2.6366],[12,-2.5895],[14,-2.5276],[15,-2.5065],[16,-2.4766],[20,-2.4066],[22,-2.3592],[23,-2.3183],[24,-2.2853],[25,-2.2357],[27,-2.1591],[29,-2.0751],[30,-2.0473],[31,-2.0243],[32,-2.0121],[33,-2.0053],[35,-2.0201],[37,-2.0594],[38,-2.0951],[39,-2.1351],[41,-2.2263],[43,-2.3354],[45,-2.4647],[47,-2.6077],[49,-2.7638],[50,-2.8372],[61,-3.729],[62,-3.7976],[64,-3.9434],[66,-4.067],[67,-4.116],[69,-4.1873],[71,-4.233],[72,-4.2442],[74,-4.2444],[81,-4.1619],[82,-4.1447],[84,-4.0691],[85,-4.0212],[86,-3.9638],[88,-3.8287],[89,-3.7818],[90,-3.75],[92,-3.7342],[93,-3.7555],[95,-3.8208],[96,-3.8604],[97,-3.9154],[98,-3.9613],[102,-4.2305],[104,-4.3318],[106,-4.4152],[107,-4.4717],[108,-4.5081],[110,-4.6031],[112,-4.6739],[117,-4.8631],[119,-4.9501],[122,-5.1048],[124,-5.2313],[126,-5.3307],[128,-5.4012],[132,-5.4571],[134,-5.4571],[135,-5.4457],[137,-5.3942],[139,-5.3184],[141,-5.2075],[144,-4.997],[152,-4.3779],[156,-4.0997],[158,-3.9747],[162,-3.7736],[164,-3.6825],[168,-3.5231],[170,-3.4655],[172,-3.4335],[173,-3.4295],[174,-3.4341],[175,-3.4487],[177,-3.5116],[178,-3.5531],[180,-3.6564],[183,-3.7809],[192,-4.0838],[193,-4.1249],[195,-4.197],[198,-4.3239],[200,-4.3948],[207,-4.7203],[210,-4.8382],[212,-4.8937],[213,-4.9096],[214,-4.8931],[215,-4.8572],[216,-4.8034],[217,-4.7238],[218,-4.6047],[219,-4.4453],[220,-4.2641],[221,-4.0362],[222,-3.7662],[223,-3.4622],[224,-3.1113],[225,-2.7221],[226,-2.2938],[227,-1.829],[228,-1.3349],[229,-0.805],[230,-0.2547],[231,0.3213],[232,0.9151],[233,1.5386],[234,2.1791],[235,2.8419],[236,3.5191],[239,5.5919],[240,6.2726],[241,6.9398],[242,7.5836],[243,8.1973],[244,8.7829],[245,9.3343],[246,9.8468],[247,10.3194],[248,10.747],[249,11.134],[250,11.463],[251,11.749],[252,11.984],[254,12.3725],[256,12.6701],[258,12.8913],[260,13.0576],[264,13.2597],[270,13.3165],[291,12.952],[294,12.7955],[297,12.5407],[299,12.2998],[308,10.9498],[309,10.8122],[323,-1.8073],[324,-2.2119],[325,-2.3238],[326,-2.425],[327,-2.5117],[328,-2.5765],[329,-2.6252],[330,-2.6553],[331,-2.6721],[332,-2.6737],[333,-2.6665],[335,-2.6366],[337,-2.5895],[339,-2.5276],[340,-2.5065],[341,-2.4766],[345,-2.4066],[347,-2.3592],[348,-2.3183],[349,-2.2853],[350,-2.2357],[352,-2.1591],[354,-2.0751],[355,-2.0473],[356,-2.0243],[357,-2.0121],[358,-2.0053],[360,-2.0201],[362,-2.0594],[363,-2.0951],[364,-2.1351],[366,-2.2263],[368,-2.3354],[370,-2.4647],[372,-2.6077],[374,-2.7638],[375,-2.8372],[386,-3.729],[387,-3.7976],[389,-3.9434],[391,-4.067],[392,-4.116],[394,-4.1873],[396,-4.233],[397,-4.2442],[399,-4.2444],[406,-4.1619],[407,-4.1447],[409,-4.0691],[410,-4.0212],[411,-3.9638],[413,-3.8287],[414,-3.7818],[415,-3.75],[417,-3.7342],[418,-3.7555],[420,-3.8208],[421,-3.8604],[422,-3.9154],[423,-3.9613],[427,-4.2305],[429,-4.3318],[431,-4.4152],[432,-4.4717],[433,-4.5081],[435,-4.6031],[437,-4.6739],[442,-4.8631],[444,-4.9501],[447,-5.1048],[449,-5.2313],[451,-5.3307],[453,-5.4012],[457,-5.4571],[459,-5.4571],[460,-5.4457],[462,-5.3942],[464,-5.3184],[466,-5.2075],[469,-4.997],[477,-4.3779],[481,-4.0997],[483,-3.9747],[487,-3.7736],[489,-3.6825],[493,-3.5231],[495,-3.4655],[497,-3.4335],[498,-3.4295],[499,-3.4341],[500,-3.4487],[502,-3.5116],[503,-3.5531],[505,-3.6564],[508,-3.7809],[517,-4.0838],[518,-4.1249],[520,-4.197],[523,-4.3239],[525,-4.3948],[532,-4.7203],[535,-4.8382],[537,-4.8937],[538,-4.9096],[539,-4.8931],[540,-4.8572],[541,-4.8034],[542,-4.7238],[543,-4.6047],[544,-4.4453],[545,-4.2641],[546,-4.0362],[547,-3.7662],[548,-3.4622],[549,-3.1113],[550,-2.7221],[551,-2.2938],[552,-1.829],[553,-1.3349],[554,-0.805],[555,-0.2547],[556,0.3213],[557,0.9151],[558,1.5386],[559,2.1791],[560,2.8419],[561,3.5191],[564,5.5919],[565,6.2726],[566,6.9398],[567,7.5836],[568,8.1973],[569,8.7829],[570,9.3343],[571,9.8468],[572,10.3194],[573,10.747],[574,11.134],[575,11.463],[576,11.749],[577,11.984],[579,12.3725],[581,12.6701],[583,12.8913],[585,13.0576],[589,13.2597],[595,13.3165],[616,12.952],[619,12.7955],[622,12.5407],[624,12.2998],[633,10.9498],[634,10.8122],[676,-1.8073],[677,-2.2119],[678,-2.3238],[679,-2.425],[680,-2.5117],[681,-2.5765],[682,-2.6252],[683,-2.6553],[684,-2.6721],[685,-2.6737],[686,-2.6665],[688,-2.6366],[690,-2.5895],[692,-2.5276],[693,-2.5065],[694,-2.4766],[698,-2.4066],[700,-2.3592],[701,-2.3183],[702,-2.2853],[703,-2.2357],[705,-2.1591],[707,-2.0751],[708,-2.0473],[709,-2.0243],[710,-2.0121],[711,-2.0053],[713,-2.0201],[715,-2.0594],[716,-2.0951],[717,-2.1351],[719,-2.2263],[721,-2.3354],[723,-2.4647],[725,-2.6077],[727,-2.7638],[728,-2.8372],[739,-3.729],[740,-3.7976],[742,-3.9434],[744,-4.067],[745,-4.116],[747,-4.1873],[749,-4.233],[750,-4.2442],[752,-4.2444],[759,-4.1619],[770,13.0592],[776,12.952],[779,12.7955],[782,12.5407],[784,12.2998],[793,10.9498],[799,10.1563],[803,9.6909],[837,-1.8073],[887,10.6772],[893,9.915],[895,9.6909],[929,-1.8073]]
       }
       hand_l_ik_ctrl.Location.Z {
        type=MovieSceneScriptingFloatChannel
        default=0
        thinned_from=751
        keys [[-2,46.1366],[20,42.9627],[28,42.3762],[67,44.3124],[84,43.3292],[93,41.9652],[102,40.1776],[107,39.6224],[112,39.5782],[120,40.345],[129,41.4695],[133,41.6591],[138,41.4752],[180,36.2425],[194,35.8629],[203,36.1974],[212,37.2159],[216,37.9923],[219,38.8691],[227,41.9776],[234,44.9605],[238,46.284],[241,46.9624],[245,47.4456],[309,48.2001],[323,46.1366],[345,42.9627],[353,42.3762],[392,44.3124],[409,43.3292],[418,41.9652],[427,40.1776],[432,39.6224],[437,39.5782],[445,40.345],[454,41.4695],[458,41.6591],[463,41.4752],[505,36.2425],[519,35.8629],[528,36.1974],[537,37.2159],[541,37.9923],[544,38.8691],[552,41.9776],[559,44.9605],[563,46.284],[566,46.9624],[570,47.4456],[634,48.2001],[676,46.1366],[698,42.9627],[706,42.3762],[745,44.3124],[755,43.9287],[759,43.607],[770,48.4203],[803,47.8977],[837,46.1366],[887,48.1678],[895,47.8977],[929,46.1366]]
       }
       hand_l_ik_ctrl.Rotation.X {
        type=MovieSceneScriptingFloatChannel
        default=0
        thinned_from=751
        keys [[-2,138.6421],[-1,139.2649],[56,144.5071],[117,140.305],[139,143.9878],[207,144.9181],[218,141.9747],[225,138.0455],[243,122.4032],[249,118.4893],[257,115.6117],[295,114.2225],[309,116.6307],[323,138.6421],[324,139.2649],[381,144.5071],[442,140.305],[464,143.9878],[532,144.9181],[543,141.9747],[550,138.0455],[568,122.4032],[574,118.4893],[582,115.6117],[620,114.2225],[634,116.6307],[677,139.2649],[734,144.5071],[759,143.1577],[770,113.9919],[803,118.3485],[837,138.6421],[887,116.8245],[895,118.3485],[929,138.6421]]
       }
       hand_l_ik_ctrl.Rotation.Y {
        type=MovieSceneScriptingFloatChannel
        default=0
        thinned_from=751
        keys [[-2,13.2302],[-1,13.5325],[1,13.7681],[4,14.2163],[9,15.1206],[11,15.4148],[23,16.645],[25,16.7992],[27,16.8802],[30,16.8566],[49,16.588],[75,17.3237],[78,17.4959],[84,18.0626],[85,18.2206],[86,18.4384],[87,18.7124],[88,19.0721],[89,19.5351],[90,20.0895],[91,20.7204],[92,21.4215],[93,22.1797],[95,23.826],[101,29.109],[103,30.7459],[105,32.1546],[107,33.3745],[109,34.3848],[111,35.1832],[113,35.7672],[115,36.1968],[118,36.5493],[123,36.5794],[133,35.7697],[143,35.8095],[168,37.5204],[174,38.3616],[182,40.1492],[191,41.0599],[202,41.2819],[211,40.6442],[213,40.3355],[215,39.8496],[217,39.1563],[219,38.1646],[221,36.9847],[223,35.6176],[225,34.0165],[227,32.1654],[229,30.0554],[230,28.9056],[231,27.6913],[233,25.1088],[235,22.3649],[237,19.5054],[240,15.0702],[242,12.075],[243,10.6037],[244,9.172],[245,7.7944],[246,6.4792],[247,5.2253],[248,4.0358],[249,2.9164],[250,1.8674],[251,0.897],[252,0.0086],[253,-0.8314],[254,-1.6245],[255,-2.3885],[256,-3.1278],[257,-3.8407],[258,-4.5291],[259,-5.1949],[260,-5.8328],[261,-6.4327],[262,-6.9971],[263,-7.5156],[264,-7.9939],[265,-8.4304],[266,-8.8234],[267,-9.1741],[269,-9.7963],[272,-10.6233],[274,-11.1029],[276,-11.5227],[278,-11.8915],[280,-12.2099],[282,-12.478],[284,-12.7078],[288,-13.0648],[290,-13.1923],[291,-13.2249],[292,-13.2258],[293,-13.1937],[295,-13.0421],[296,-12.9204],[297,-12.7658],[298,-12.5811],[300,-12.131],[302,-11.5932],[304,-10.9998],[308,-9.7333],[309,-9.4311],[323,13.2302],[324,13.5325],[326,13.7681],[329,14.2163],[334,15.1206],[336,15.4148],[348,16.645],[350,16.7992],[352,16.8802],[355,16.8566],[374,16.588],[400,17.3237],[403,17.4959],[409,18.0626],[410,18.2206],[411,18.4384],[412,18.7124],[413,19.0721],[414,19.5351],[415,20.0895],[416,20.7204],[417,21.4215],[418,22.1797],[420,23.826],[426,29.109],[428,30.7459],[430,32.1546],[432,33.3745],[434,34.3848],[436,35.1832],[438,35.7672],[440,36.1968],[443,36.5493],[448,36.5794],[458,35.7697],[468,35.8095],[493,37.5204],[499,38.3616],[507,40.1492],[516,41.0599],[527,41.2819],[536,40.6442],[538,40.3355],[540,39.8496],[542,39.1563],[544,38.1646],[546,36.9847],[548,35.6176],[550,34.0165],[552,32.1654],[554,30.0554],[555,28.9056],[556,27.6913],[558,25.1088],[560,22.3649],[562,19.5054],[565,15.0702],[567,12.075],[568,10.6037],[569,9.172],[570,7.7944],[571,6.4792],[572,5.2253],[573,4.0358],[574,2.9164],[575,1.8674],[576,0.897],[577,0.0086],[578,-0.8314],[579,-1.6245],[580,-2.3885],[581,-3.1278],[582,-3.8407],[583,-4.5291],[584,-5.1949],[585,-5.8328],[586,-6.4327],[587,-6.9971],[588,-7.5156],[589,-7.9939],[590,-8.4304],[591,-8.8234],[592,-9.1741],[594,-9.7963],[597,-10.6233],[599,-11.1029],[601,-11.5227],[603,-11.8915],[605,-12.2099],[607,-12.478],[609,-12.7078],[613,-13.0648],[615,-13.1923],[616,-13.2249],[617,-13.2258],[618,-13.1937],[620,-13.0421],[621,-12.9204],[622,-12.7658],[623,-12.5811],[625,-12.131],[627,-11.5932],[629,-10.9998],[633,-9.7333],[634,-9.4311],[676,13.2302],[677,13.5325],[679,13.7681],[682,14.2163],[687,15.1206],[689,15.4148],[701,16.645],[703,16.7992],[705,16.8802],[708,16.8566],[727,16.588],[753,17.3237],[755,17.4277],[757,17.577],[759,17.7642],[770,-12.8028],[773,-13.0648],[775,-13.1923],[776,-13.2249],[777,-13.2258],[778,-13.1937],[780,-13.0421],[781,-12.9204],[782,-12.7658],[783,-12.5811],[785,-12.131],[787,-11.5932],[789,-10.9998],[793,-9.7333],[795,-9.137],[799,-8.036],[801,-7.5338],[803,-7.0753],[837,13.2302],[887,-9.137],[891,-8.036],[893,-7.5338],[895,-7.0753],[929,13.2302]]
       }
       hand_l_ik_ctrl.Rotation.Z {
        type=MovieSceneScriptingFloatChannel
        default=0
        thinned_from=751
        keys [[-2,134.7565],[25,129.9734],[90,128.8242],[114,126.6987],[216,129.6547],[259,144.8502],[293,147.1182],[309,143.8077],[323,134.7565],[350,129.9734],[415,128.8242],[439,126.6987],[541,129.6547],[584,144.8502],[618,147.1182],[676,134.7565],[703,129.9734],[759,129.4009],[770,147.3],[803,141.62],[837,134.7565],[887,143.5662],[895,141.62],[929,134.7565]]
       }
       hand_l_ik_ctrl.Scale.X {type=MovieSceneScriptingFloatChannel, default=1, const [1,751]}
       hand_l_ik_ctrl.Scale.Y {type=MovieSceneScriptingFloatChannel, default=1, const [1,751]}
       hand_l_ik_ctrl.Scale.Z {type=MovieSceneScriptingFloatChannel, default=1, const [1,751]}
       Softness {type=MovieSceneScriptingFloatChannel, default=0, const [0,751]}
       hand_l_ikorient_ctrl.Rotation.Roll {type=MovieSceneScriptingFloatChannel, default=0, const [0,445]}
       hand_l_ikorient_ctrl.Rotation.Pitch {type=MovieSceneScriptingFloatChannel, default=0, const [0,445]}
       hand_l_ikorient_ctrl.Rotation.Yaw {type=MovieSceneScriptingFloatChannel, default=0, const [0,445]}
       arm_l_pv_ik_ctrl.Location.X {
        type=MovieSceneScriptingFloatChannel
        default=0
        thinned_from=751
        keys [[-2,12.918],[-1,13.0569],[3,13.2285],[11,13.3549],[19,13.6232],[22,13.6504],[38,13.3223],[46,13.4019],[50,13.5329],[74,14.8527],[82,15.0699],[84,15.0756],[86,14.9943],[87,14.9085],[89,14.6442],[91,14.2715],[93,13.8],[94,13.5344],[97,12.6137],[100,11.5618],[104,10.0959],[106,9.4058],[108,8.7792],[109,8.4935],[111,7.981],[112,7.7571],[114,7.3649],[115,7.181],[117,6.862],[118,6.7184],[119,6.5893],[122,6.2899],[124,6.1353],[126,6.0093],[130,5.8083],[137,5.4031],[152,4.6393],[163,4.1121],[167,3.8858],[171,3.6377],[174,3.4189],[176,3.2538],[178,3.0673],[179,2.9852],[181,2.8358],[183,2.7258],[185,2.656],[187,2.6152],[189,2.5895],[191,2.5888],[194,2.6315],[195,2.6562],[197,2.7192],[199,2.7958],[200,2.8385],[202,2.9511],[204,3.0856],[205,3.1577],[207,3.3227],[209,3.5117],[211,3.7246],[212,3.8457],[214,4.1218],[215,4.2703],[216,4.4278],[217,4.5993],[220,5.1476],[223,5.6394],[228,6.3545],[232,6.9134],[233,7.0367],[234,7.1441],[235,7.2251],[236,7.2821],[237,7.3078],[238,7.3136],[239,7.2835],[240,7.2348],[241,7.1556],[242,7.0577],[244,6.8087],[249,6.0544],[251,5.7915],[255,5.3587],[257,5.1584],[260,4.9081],[264,4.6145],[267,4.4412],[270,4.2953],[271,4.2622],[273,4.1719],[276,4.0616],[279,3.9951],[280,3.9887],[281,3.9707],[284,3.9554],[285,3.9612],[286,3.9556],[287,3.9725],[288,3.98],[290,4.0196],[291,4.0576],[293,4.2023],[294,4.3051],[295,4.423],[296,4.5636],[297,4.7187],[299,5.0941],[302,5.7503],[308,7.1445],[309,7.3605],[323,12.918],[324,13.0569],[328,13.2285],[336,13.3549],[344,13.6232],[347,13.6504],[363,13.3223],[371,13.4019],[375,13.5329],[399,14.8527],[407,15.0699],[409,15.0756],[411,14.9943],[412,14.9085],[414,14.6442],[416,14.2715],[418,13.8],[419,13.5344],[422,12.6137],[425,11.5618],[429,10.0959],[431,9.4058],[433,8.7792],[434,8.4935],[436,7.981],[437,7.7571],[439,7.3649],[440,7.181],[442,6.862],[443,6.7184],[444,6.5893],[447,6.2899],[449,6.1353],[451,6.0093],[455,5.8083],[462,5.4031],[477,4.6393],[488,4.1121],[492,3.8858],[496,3.6377],[499,3.4189],[501,3.2538],[503,3.0673],[504,2.9852],[506,2.8358],[508,2.7258],[510,2.656],[512,2.6152],[514,2.5895],[516,2.5888],[519,2.6315],[520,2.6562],[522,2.7192],[524,2.7958],[525,2.8385],[527,2.9511],[529,3.0856],[530,3.1577],[532,3.3227],[534,3.5117],[536,3.7246],[537,3.8457],[539,4.1218],[540,4.2703],[541,4.4278],[542,4.5993],[545,5.1476],[548,5.6394],[553,6.3545],[557,6.9134],[558,7.0367],[559,7.1441],[560,7.2251],[561,7.2821],[562,7.3078],[563,7.3136],[564,7.2835],[565,7.2348],[566,7.1556],[567,7.0577],[569,6.8087],[574,6.0544],[576,5.7915],[580,5.3587],[582,5.1584],[585,4.9081],[589,4.6145],[592,4.4412],[595,4.2953],[596,4.2622],[598,4.1719],[601,4.0616],[604,3.9951],[605,3.9887],[606,3.9707],[609,3.9554],[610,3.9612],[611,3.9556],[612,3.9725],[613,3.98],[615,4.0196],[616,4.0576],[618,4.2023],[619,4.3051],[620,4.423],[621,4.5636],[622,4.7187],[624,5.0941],[627,5.7503],[633,7.1445],[634,7.3605],[677,13.0569],[681,13.2285],[689,13.3549],[697,13.6232],[700,13.6504],[716,13.3223],[724,13.4019],[728,13.5329],[752,14.8527],[757,15.0204],[759,15.0566],[770,3.9612],[771,3.9556],[772,3.9725],[773,3.98],[775,4.0196],[776,4.0576],[778,4.2023],[779,4.3051],[780,4.423],[781,4.5636],[782,4.7187],[784,5.0941],[787,5.7503],[793,7.1445],[797,7.9916],[801,8.7853],[803,9.1471],[837,12.918],[887,7.5748],[891,8.3962],[895,9.1471],[929,12.918]]
       }
       arm_l_pv_ik_ctrl.Location.Y {
        type=MovieSceneScriptingFloatChannel
        default=0
        thinned_from=751
        keys [[-2,28.9734],[-1,28.3882],[8,26.1896],[11,25.6194],[19,24.6494],[25,24.2274],[28,24.1796],[59,25.0038],[77,24.7657],[84,24.2686],[85,24.1258],[86,23.9304],[87,23.6831],[88,23.3659],[89,22.9644],[90,22.4986],[91,21.9612],[92,21.376],[94,20.0785],[101,15.0523],[102,14.3727],[103,13.7412],[104,13.1544],[105,12.6283],[106,12.1337],[107,11.685],[108,11.2905],[109,10.9295],[110,10.608],[111,10.33],[112,10.0928],[113,9.8936],[114,9.7227],[115,9.589],[116,9.4983],[117,9.4319],[118,9.3901],[119,9.3769],[120,9.3834],[122,9.4422],[124,9.5323],[126,9.6597],[129,9.8939],[130,9.9539],[132,10.0081],[133,10.0152],[135,9.9553],[136,9.9011],[138,9.7385],[140,9.5197],[142,9.271],[153,7.6791],[157,7.1571],[160,6.809],[171,5.6743],[174,5.3201],[175,5.1889],[179,4.5897],[180,4.455],[181,4.3359],[182,4.245],[183,4.1626],[184,4.113],[186,4.0446],[188,4.0016],[190,3.9868],[192,3.9911],[194,4.0097],[196,4.0558],[198,4.1208],[200,4.214],[202,4.3395],[203,4.4128],[205,4.5892],[207,4.8027],[209,5.0516],[211,5.3437],[212,5.5169],[213,5.7175],[214,5.9565],[215,6.2414],[216,6.5771],[217,6.9787],[218,7.466],[219,8.0205],[220,8.6291],[221,9.2731],[222,9.9616],[223,10.6947],[224,11.4892],[225,12.3305],[226,13.221],[227,14.1527],[228,15.147],[229,16.1741],[230,17.2494],[231,18.3726],[233,20.72],[242,31.8865],[244,34.2198],[246,36.3488],[248,38.2791],[250,39.9721],[252,41.4209],[256,43.881],[260,45.969],[263,47.253],[266,48.2412],[277,50.4958],[290,51.6495],[295,51.5159],[300,50.8081],[309,48.7502],[323,28.9734],[324,28.3882],[333,26.1896],[336,25.6194],[344,24.6494],[350,24.2274],[353,24.1796],[384,25.0038],[402,24.7657],[409,24.2686],[410,24.1258],[411,23.9304],[412,23.6831],[413,23.3659],[414,22.9644],[415,22.4986],[416,21.9612],[417,21.376],[419,20.0785],[426,15.0523],[427,14.3727],[428,13.7412],[429,13.1544],[430,12.6283],[431,12.1337],[432,11.685],[433,11.2905],[434,10.9295],[435,10.608],[436,10.33],[437,10.0928],[438,9.8936],[439,9.7227],[440,9.589],[441,9.4983],[442,9.4319],[443,9.3901],[444,9.3769],[445,9.3834],[447,9.4422],[449,9.5323],[451,9.6597],[454,9.8939],[455,9.9539],[457,10.0081],[458,10.0152],[460,9.9553],[461,9.9011],[463,9.7385],[465,9.5197],[467,9.271],[478,7.6791],[482,7.1571],[485,6.809],[496,5.6743],[499,5.3201],[500,5.1889],[504,4.5897],[505,4.455],[506,4.3359],[507,4.245],[508,4.1626],[509,4.113],[511,4.0446],[513,4.0016],[515,3.9868],[517,3.9911],[519,4.0097],[521,4.0558],[523,4.1208],[525,4.214],[527,4.3395],[528,4.4128],[530,4.5892],[532,4.8027],[534,5.0516],[536,5.3437],[537,5.5169],[538,5.7175],[539,5.9565],[540,6.2414],[541,6.5771],[542,6.9787],[543,7.466],[544,8.0205],[545,8.6291],[546,9.2731],[547,9.9616],[548,10.6947],[549,11.4892],[550,12.3305],[551,13.221],[552,14.1527],[553,15.147],[554,16.1741],[555,17.2494],[556,18.3726],[558,20.72],[567,31.8865],[569,34.2198],[571,36.3488],[573,38.2791],[575,39.9721],[577,41.4209],[581,43.881],[585,45.969],[588,47.253],[591,48.2412],[602,50.4958],[615,51.6495],[620,51.5159],[625,50.8081],[634,48.7502],[676,28.9734],[677,28.3882],[686,26.1896],[689,25.6194],[697,24.6494],[703,24.2274],[706,24.1796],[737,25.0038],[755,24.7657],[759,24.5269],[770,51.3595],[777,51.6694],[782,51.2996],[803,46.9059],[837,28.9734],[887,48.5266],[895,46.9059],[929,28.9734]]
       }
       arm_l_pv_ik_ctrl.Location.Z {
        type=MovieSceneScriptingFloatChannel
        default=0
        thinned_from=751
        keys [[-2,37.3524],[-1,37.7692],[2,37.9498],[5,37.7967],[26,34.2147],[29,33.9129],[33,33.9061],[39,34.3573],[62,36.9945],[67,37.2364],[72,36.9799],[92,33.9194],[96,33.568],[100,33.5566],[105,34.01],[110,34.8884],[124,38.3475],[128,39.1109],[134,39.6843],[139,39.7601],[182,34.3471],[195,33.5432],[202,33.5729],[212,34.429],[217,34.9125],[221,34.975],[230,34.5572],[244,35.0117],[258,36.2471],[278,38.4907],[293,39.1508],[304,38.7899],[309,38.3233],[323,37.3524],[324,37.7692],[327,37.9498],[330,37.7967],[351,34.2147],[354,33.9129],[358,33.9061],[364,34.3573],[387,36.9945],[392,37.2364],[397,36.9799],[417,33.9194],[421,33.568],[425,33.5566],[430,34.01],[435,34.8884],[449,38.3475],[453,39.1109],[459,39.6843],[464,39.7601],[507,34.3471],[520,33.5432],[527,33.5729],[537,34.429],[542,34.9125],[546,34.975],[555,34.5572],[569,35.0117],[583,36.2471],[603,38.4907],[618,39.1508],[629,38.7899],[634,38.3233],[676,37.3524],[677,37.7692],[680,37.9498],[683,37.7967],[704,34.2147],[707,33.9129],[711,33.9061],[717,34.3573],[740,36.9945],[745,37.2364],[750,36.9799],[755,36.3124],[759,35.6418],[770,38.9286],[785,39.0055],[803,37.5082],[837,37.3524],[887,38.2191],[895,37.5082],[929,37.3524]]
       }
       arm_l_pv_ik_ctrl.Rotation.X {
        type=MovieSceneScriptingFloatChannel
        default=0
        thinned_from=751
        keys [[-2,-21.2312],[-1,-21.4583],[5,-21.7297],[17,-21.7819],[39,-22.9933],[50,-23.1458],[60,-22.7088],[77,-21.6717],[82,-21.5098],[85,-21.2789],[88,-20.8441],[95,-19.5094],[104,-17.4893],[106,-17.1097],[109,-16.6854],[112,-16.369],[114,-16.2365],[116,-16.2221],[118,-16.3279],[120,-16.5314],[122,-16.8213],[131,-18.5764],[133,-18.8895],[135,-19.1142],[137,-19.241],[142,-19.2633],[157,-18.68],[181,-18.7941],[193,-19.061],[202,-18.9764],[207,-18.7543],[210,-18.5044],[219,-17.3676],[229,-16.0664],[232,-15.5751],[238,-14.4089],[240,-14.1023],[248,-13.2352],[251,-13.0158],[253,-12.9405],[266,-12.901],[273,-13.0592],[279,-13.2746],[286,-13.3276],[292,-13.2132],[309,-12.491],[323,-21.2312],[324,-21.4583],[330,-21.7297],[342,-21.7819],[364,-22.9933],[375,-23.1458],[385,-22.7088],[402,-21.6717],[407,-21.5098],[410,-21.2789],[413,-20.8441],[420,-19.5094],[429,-17.4893],[431,-17.1097],[434,-16.6854],[437,-16.369],[439,-16.2365],[441,-16.2221],[443,-16.3279],[445,-16.5314],[447,-16.8213],[456,-18.5764],[458,-18.8895],[460,-19.1142],[462,-19.241],[467,-19.2633],[482,-18.68],[506,-18.7941],[518,-19.061],[527,-18.9764],[532,-18.7543],[535,-18.5044],[544,-17.3676],[554,-16.0664],[557,-15.5751],[563,-14.4089],[565,-14.1023],[573,-13.2352],[576,-13.0158],[578,-12.9405],[591,-12.901],[598,-13.0592],[604,-13.2746],[611,-13.3276],[617,-13.2132],[634,-12.491],[677,-21.4583],[683,-21.7297],[695,-21.7819],[717,-22.9933],[728,-23.1458],[738,-22.7088],[755,-21.6717],[759,-21.5528],[770,-13.3272],[775,-13.2695],[800,-12.2157],[803,-12.0548],[837,-21.2312],[887,-12.4493],[895,-12.0548],[929,-21.2312]]
       }
       arm_l_pv_ik_ctrl.Rotation.Y {
        type=MovieSceneScriptingFloatChannel
        default=0
        thinned_from=751
        keys [[-2,12.9791],[-1,13.4401],[1,13.8279],[4,14.3043],[9,14.9037],[12,15.4199],[18,16.6994],[21,17.2351],[23,17.501],[29,17.9661],[40,18.296],[52,18.3256],[82,17.3671],[84,17.2501],[86,17.0381],[88,16.7263],[91,16.0888],[94,15.2976],[97,14.3809],[105,11.6913],[107,11.0727],[109,10.5338],[110,10.2933],[111,10.0796],[113,9.7184],[115,9.4355],[116,9.3231],[117,9.2379],[118,9.1773],[119,9.1483],[120,9.1481],[121,9.173],[123,9.2784],[128,9.7179],[130,9.8495],[133,9.9592],[140,10.0176],[154,9.8226],[162,9.5909],[165,9.4482],[177,8.6164],[181,8.4197],[189,8.1799],[203,7.9613],[206,7.8655],[209,7.7063],[211,7.5535],[213,7.3597],[214,7.2324],[215,7.0766],[216,6.8986],[217,6.6938],[218,6.4576],[219,6.1882],[220,5.8841],[221,5.5404],[222,5.1618],[224,4.3503],[226,3.529],[227,3.1344],[228,2.7647],[229,2.4234],[230,2.1259],[231,1.8814],[232,1.685],[233,1.5208],[234,1.3827],[235,1.2598],[236,1.155],[237,1.0623],[238,1.0011],[239,0.9781],[240,0.9986],[241,1.0657],[242,1.1721],[243,1.3022],[245,1.5783],[247,1.8509],[249,2.1415],[253,2.7787],[254,2.927],[255,3.0869],[256,3.2582],[257,3.4427],[258,3.6422],[261,4.2837],[262,4.4873],[263,4.6701],[264,4.8428],[265,5.0018],[266,5.1495],[267,5.2813],[275,6.2398],[280,6.7697],[283,7.0402],[285,7.1905],[287,7.3195],[290,7.4626],[293,7.5584],[308,7.7802],[309,7.8059],[323,12.9791],[324,13.4401],[326,13.8279],[329,14.3043],[334,14.9037],[337,15.4199],[343,16.6994],[346,17.2351],[348,17.501],[354,17.9661],[365,18.296],[377,18.3256],[407,17.3671],[409,17.2501],[411,17.0381],[413,16.7263],[416,16.0888],[419,15.2976],[422,14.3809],[430,11.6913],[432,11.0727],[434,10.5338],[435,10.2933],[436,10.0796],[438,9.7184],[440,9.4355],[441,9.3231],[442,9.2379],[443,9.1773],[444,9.1483],[445,9.1481],[446,9.173],[448,9.2784],[453,9.7179],[455,9.8495],[458,9.9592],[465,10.0176],[479,9.8226],[487,9.5909],[490,9.4482],[502,8.6164],[506,8.4197],[514,8.1799],[528,7.9613],[531,7.8655],[534,7.7063],[536,7.5535],[538,7.3597],[539,7.2324],[540,7.0766],[541,6.8986],[542,6.6938],[543,6.4576],[544,6.1882],[545,5.8841],[546,5.5404],[547,5.1618],[549,4.3503],[551,3.529],[552,3.1344],[553,2.7647],[554,2.4234],[555,2.1259],[556,1.8814],[557,1.685],[558,1.5208],[559,1.3827],[560,1.2598],[561,1.155],[562,1.0623],[563,1.0011],[564,0.9781],[565,0.9986],[566,1.0657],[567,1.1721],[568,1.3022],[570,1.5783],[572,1.8509],[574,2.1415],[578,2.7787],[579,2.927],[580,3.0869],[581,3.2582],[582,3.4427],[583,3.6422],[586,4.2837],[587,4.4873],[588,4.6701],[589,4.8428],[590,5.0018],[591,5.1495],[592,5.2813],[600,6.2398],[605,6.7697],[608,7.0402],[610,7.1905],[612,7.3195],[615,7.4626],[618,7.5584],[633,7.7802],[634,7.8059],[676,12.9791],[677,13.4401],[679,13.8279],[682,14.3043],[687,14.9037],[690,15.4199],[696,16.6994],[699,17.2351],[701,17.501],[707,17.9661],[718,18.296],[730,18.3256],[759,17.4037],[770,7.1905],[772,7.3195],[775,7.4626],[778,7.5584],[793,7.7802],[796,7.8682],[799,8.0138],[803,8.2851],[837,12.9791],[887,7.8328],[890,7.9589],[895,8.2851],[929,12.9791]]
       }
       arm_l_pv_ik_ctrl.Rotation.Z {
        type=MovieSceneScriptingFloatChannel
        default=0
        thinned_from=751
        keys [[-2,45.3929],[-1,44.6682],[20,39.0246],[25,38.0407],[28,37.7316],[77,37.6587],[85,37.0987],[87,36.7864],[88,36.5289],[89,36.1723],[90,35.7253],[92,34.6064],[94,33.2631],[101,28.0541],[103,26.7148],[105,25.6111],[107,24.6632],[109,23.8851],[111,23.2856],[113,22.8663],[115,22.5474],[119,22.1573],[125,21.8038],[148,21.3959],[167,20.3074],[171,19.9213],[174,19.4684],[176,19.0424],[180,18.0135],[182,17.5747],[183,17.3915],[187,16.8978],[191,16.5544],[196,16.3365],[199,16.3139],[202,16.4046],[205,16.6295],[207,16.8665],[209,17.1837],[211,17.5885],[212,17.8375],[213,18.1419],[214,18.51],[215,18.9505],[216,19.4497],[217,20.0383],[218,20.7342],[219,21.5215],[220,22.3758],[221,23.3027],[222,24.3133],[223,25.3948],[224,26.5577],[225,27.7966],[226,29.1109],[227,30.499],[229,33.48],[231,36.693],[234,41.8425],[242,56.3625],[245,61.3885],[247,64.4048],[249,67.0933],[251,69.4068],[255,73.1117],[259,76.1176],[263,78.4705],[267,80.1431],[278,82.6816],[291,83.6541],[296,83.1441],[304,80.4908],[309,78.3866],[323,45.3929],[324,44.6682],[345,39.0246],[350,38.0407],[353,37.7316],[402,37.6587],[410,37.0987],[412,36.7864],[413,36.5289],[414,36.1723],[415,35.7253],[417,34.6064],[419,33.2631],[426,28.0541],[428,26.7148],[430,25.6111],[432,24.6632],[434,23.8851],[436,23.2856],[438,22.8663],[440,22.5474],[444,22.1573],[450,21.8038],[473,21.3959],[492,20.3074],[496,19.9213],[499,19.4684],[501,19.0424],[505,18.0135],[507,17.5747],[508,17.3915],[512,16.8978],[516,16.5544],[521,16.3365],[524,16.3139],[527,16.4046],[530,16.6295],[532,16.8665],[534,17.1837],[536,17.5885],[537,17.8375],[538,18.1419],[539,18.51],[540,18.9505],[541,19.4497],[542,20.0383],[543,20.7342],[544,21.5215],[545,22.3758],[546,23.3027],[547,24.3133],[548,25.3948],[549,26.5577],[550,27.7966],[551,29.1109],[552,30.499],[554,33.48],[556,36.693],[559,41.8425],[567,56.3625],[570,61.3885],[572,64.4048],[574,67.0933],[576,69.4068],[580,73.1117],[584,76.1176],[588,78.4705],[592,80.1431],[603,82.6816],[616,83.6541],[621,83.1441],[629,80.4908],[634,78.3866],[676,45.3929],[677,44.6682],[698,39.0246],[703,38.0407],[706,37.7316],[755,37.6587],[759,37.3792],[770,83.366],[777,83.626],[782,82.9226],[803,75.0141],[837,45.3929],[887,77.9862],[895,75.0141],[929,45.3929]]
       }
       arm_l_pv_ik_ctrl.Scale.X {type=MovieSceneScriptingFloatChannel, default=1, const [1,751]}
       arm_l_pv_ik_ctrl.Scale.Y {type=MovieSceneScriptingFloatChannel, default=1, const [1,751]}
       arm_l_pv_ik_ctrl.Scale.Z {type=MovieSceneScriptingFloatChannel, default=1, const [1,751]}
       upperarm_r_fk_ctrl.Location.X {type=MovieSceneScriptingFloatChannel, default=0, thinned_from=615, keys [[-2,0],[929,0]]}
       upperarm_r_fk_ctrl.Location.Y {type=MovieSceneScriptingFloatChannel, default=0, thinned_from=615, keys [[-2,0],[929,0]]}
       upperarm_r_fk_ctrl.Location.Z {type=MovieSceneScriptingFloatChannel, default=0, thinned_from=615, keys [[-2,0],[929,0]]}
       upperarm_r_fk_ctrl.Rotation.X {
        type=MovieSceneScriptingFloatChannel
        default=69.0489
        thinned_from=591
        keys [[-2,67.5144],[-1,67.188],[6,67.0547],[10,67.1841],[22,65.7222],[82,67.6026],[87,68.5229],[98,72.2503],[102,73.0206],[106,73.0967],[113,71.9295],[122,69.6713],[171,67.9914],[193,68.5952],[211,66.8124],[216,65.6577],[219,64.5774],[233,56.9821],[236,55.7627],[240,54.6481],[244,54.0386],[275,53.9099],[292,55.514],[296,56.1983],[310,58.156],[326,61.7201],[348,65.6868],[407,67.6026],[412,68.5229],[423,72.2503],[427,73.0206],[431,73.0967],[438,71.9295],[447,69.6713],[496,67.9914],[518,68.5952],[536,66.8124],[541,65.6577],[544,64.5774],[558,56.9821],[561,55.7627],[565,54.6481],[569,54.0386],[600,53.9099],[617,55.514],[621,56.1983],[635,58.156],[645,44.0216],[666,15.4556],[679,17.2148],[698,17.1917],[787,15.0617],[830,16.6283],[841,17.211],[842,17.2148],[857,8.6428],[861,7.7797],[880,64.8528],[888,61.2295],[903,63.0772],[929,67.5144]]
       }
       upperarm_r_fk_ctrl.Rotation.Y {
        type=MovieSceneScriptingFloatChannel
        default=-11.0059
        thinned_from=598
        keys [[-2,-9.6673],[-1,-9.8726],[1,-10.255],[4,-10.9152],[7,-11.6538],[9,-12.2032],[11,-12.8267],[16,-14.507],[19,-15.4204],[21,-15.9359],[23,-16.351],[25,-16.6481],[27,-16.8426],[29,-16.9368],[50,-16.85],[79,-17.7469],[81,-17.7677],[82,-17.8199],[83,-17.9467],[84,-18.1512],[85,-18.4568],[86,-18.8597],[87,-19.3378],[88,-19.8994],[89,-20.5424],[93,-23.3965],[94,-24.0421],[95,-24.6041],[96,-25.0561],[97,-25.3838],[98,-25.6002],[99,-25.7172],[100,-25.7339],[101,-25.655],[102,-25.4811],[103,-25.2099],[104,-24.8438],[106,-23.9251],[111,-21.2938],[116,-18.8706],[117,-18.4193],[118,-18.0217],[119,-17.6908],[120,-17.4374],[121,-17.2506],[122,-17.1071],[123,-17.0173],[124,-16.9904],[125,-17.0159],[127,-17.1859],[132,-17.7263],[145,-18.6856],[157,-19.2212],[169,-19.1866],[174,-19.1364],[177,-19.2724],[181,-19.5242],[183,-19.5638],[196,-18.9894],[202,-18.5168],[207,-17.9347],[211,-17.3336],[213,-16.944],[214,-16.6882],[215,-16.3833],[216,-16.0336],[217,-15.6306],[218,-15.1506],[219,-14.5801],[220,-13.9081],[221,-13.1555],[222,-12.3463],[223,-11.5037],[226,-8.9104],[227,-8.089],[228,-7.3169],[229,-6.6186],[230,-6.0096],[231,-5.5072],[232,-5.1204],[233,-4.8553],[234,-4.6869],[235,-4.5887],[236,-4.5514],[237,-4.5791],[238,-4.663],[239,-4.7942],[240,-4.9725],[241,-5.1917],[242,-5.4507],[243,-5.7432],[244,-6.0699],[245,-6.4248],[246,-6.8054],[249,-8.017],[254,-10.1109],[257,-11.2916],[259,-12.021],[260,-12.3491],[261,-12.6482],[262,-12.9034],[263,-13.1175],[264,-13.2824],[265,-13.3982],[266,-13.4704],[269,-13.457],[284,-12.9931],[290,-12.9049],[293,-12.9692],[295,-13.0673],[296,-13.1432],[343,-15.1298],[345,-15.6904],[347,-16.1575],[349,-16.513],[351,-16.7581],[353,-16.9014],[355,-16.9556],[375,-16.85],[404,-17.7469],[406,-17.7677],[407,-17.8199],[408,-17.9467],[409,-18.1512],[410,-18.4568],[411,-18.8597],[412,-19.3378],[413,-19.8994],[414,-20.5424],[418,-23.3965],[419,-24.0421],[420,-24.6041],[421,-25.0561],[422,-25.3838],[423,-25.6002],[424,-25.7172],[425,-25.7339],[426,-25.655],[427,-25.4811],[428,-25.2099],[429,-24.8438],[431,-23.9251],[436,-21.2938],[441,-18.8706],[442,-18.4193],[443,-18.0217],[444,-17.6908],[445,-17.4374],[446,-17.2506],[447,-17.1071],[448,-17.0173],[449,-16.9904],[450,-17.0159],[452,-17.1859],[457,-17.7263],[470,-18.6856],[482,-19.2212],[494,-19.1866],[499,-19.1364],[502,-19.2724],[506,-19.5242],[508,-19.5638],[521,-18.9894],[527,-18.5168],[532,-17.9347],[536,-17.3336],[538,-16.944],[539,-16.6882],[540,-16.3833],[541,-16.0336],[542,-15.6306],[543,-15.1506],[544,-14.5801],[545,-13.9081],[546,-13.1555],[547,-12.3463],[548,-11.5037],[551,-8.9104],[552,-8.089],[553,-7.3169],[554,-6.6186],[555,-6.0096],[556,-5.5072],[557,-5.1204],[558,-4.8553],[559,-4.6869],[560,-4.5887],[561,-4.5514],[562,-4.5791],[563,-4.663],[564,-4.7942],[565,-4.9725],[566,-5.1917],[567,-5.4507],[568,-5.7432],[569,-6.0699],[570,-6.4248],[571,-6.8054],[574,-8.017],[579,-10.1109],[582,-11.2916],[584,-12.021],[585,-12.3491],[586,-12.6482],[587,-12.9034],[588,-13.1175],[589,-13.2824],[590,-13.3982],[591,-13.4704],[594,-13.457],[609,-12.9931],[615,-12.9049],[618,-12.9692],[620,-13.0673],[621,-13.1432],[638,-15.4176],[657,-2.8667],[679,-1.2099],[702,-2.924],[735,-3.3004],[781,3.8302],[795,8.1592],[841,-1.2104],[842,-1.2099],[887,-15.6291],[893,-16.798],[895,-17.1178],[903,-16.0745],[929,-9.6673]]
       }
       upperarm_r_fk_ctrl.Rotation.Z {
        type=MovieSceneScriptingFloatChannel
        default=52.7544
        thinned_from=604
        keys [[-2,55.0483],[8,56.4139],[11,57.1698],[22,61.5642],[27,62.8324],[47,63.793],[85,60.5358],[95,60.6435],[98,60.1729],[102,58.9243],[111,55.3364],[114,54.4727],[116,54.1408],[118,54.0473],[121,54.3503],[130,56.5568],[134,57.0937],[171,54.9316],[202,56.2781],[212,55.76],[219,54.6715],[221,54.1323],[231,49.7994],[238,47.7935],[247,46.2595],[254,45.8314],[289,47.0461],[296,46.8319],[336,57.1698],[347,61.5642],[352,62.8324],[372,63.793],[410,60.5358],[420,60.6435],[423,60.1729],[427,58.9243],[436,55.3364],[439,54.4727],[441,54.1408],[443,54.0473],[446,54.3503],[455,56.5568],[459,57.0937],[496,54.9316],[527,56.2781],[537,55.76],[544,54.6715],[546,54.1323],[556,49.7994],[563,47.7935],[572,46.2595],[579,45.8314],[614,47.0461],[621,46.8319],[638,46.8524],[651,20.3937],[663,-19.6301],[667,-25.8165],[682,-21.2861],[703,-25.5092],[726,-24.4264],[785,-28.6332],[827,-22.208],[841,-22.1995],[852,-20.5182],[861,-6.9203],[881,50.2928],[895,50.2271],[903,48.3628],[929,55.0483]]
       }
       upperarm_r_fk_ctrl.Scale.X {type=MovieSceneScriptingFloatChannel, default=1, const [1,615]}
       upperarm_r_fk_ctrl.Scale.Y {type=MovieSceneScriptingFloatChannel, default=1, const [1,615]}
       upperarm_r_fk_ctrl.Scale.Z {type=MovieSceneScriptingFloatChannel, default=1, const [1,615]}
       lowerarm_r_fk_ctrl.Location.X {type=MovieSceneScriptingFloatChannel, default=0, thinned_from=712, keys [[-2,0],[929,0]]}
       lowerarm_r_fk_ctrl.Location.Y {type=MovieSceneScriptingFloatChannel, default=0, thinned_from=712, keys [[-2,0],[929,0]]}
       lowerarm_r_fk_ctrl.Location.Z {type=MovieSceneScriptingFloatChannel, default=0, thinned_from=712, keys [[-2,0],[929,0]]}
       lowerarm_r_fk_ctrl.Rotation.X {
        type=MovieSceneScriptingFloatChannel
        default=0
        thinned_from=712
        keys [[-2,-0.7327],[-1,-0.7471],[0,-0.7507],[1,-0.7579],[5,-0.7723],[8,-0.7723],[9,-0.7759],[10,-0.7759],[11,-0.7723],[13,-0.7723],[14,-0.7687],[15,-0.7702],[18,-0.7702],[19,-0.7676],[22,-0.7676],[23,-0.7702],[26,-0.7702],[27,-0.7728],[29,-0.7728],[30,-0.7753],[31,-0.7745],[33,-0.7779],[35,-0.7779],[36,-0.7813],[39,-0.7813],[41,-0.7848],[46,-0.7848],[49,-0.7862],[54,-0.779],[55,-0.779],[63,-0.7623],[64,-0.7623],[65,-0.7579],[66,-0.7579],[67,-0.7535],[68,-0.7535],[69,-0.7491],[70,-0.7491],[71,-0.7447],[72,-0.7447],[73,-0.7403],[74,-0.7403],[75,-0.7359],[77,-0.7183],[79,-0.6925],[81,-0.6523],[82,-0.6255],[84,-0.5652],[88,-0.4312],[89,-0.401],[91,-0.3474],[94,-0.277],[95,-0.2609],[96,-0.2405],[98,-0.2134],[100,-0.1998],[101,-0.1862],[102,-0.1794],[103,-0.1794],[104,-0.1726],[108,-0.1726],[110,-0.1862],[111,-0.1916],[112,-0.2024],[113,-0.2149],[114,-0.2293],[116,-0.2651],[118,-0.2992],[119,-0.3171],[120,-0.3332],[123,-0.3763],[125,-0.3996],[126,-0.4103],[127,-0.4175],[129,-0.4399],[136,-0.479],[137,-0.479],[138,-0.4846],[139,-0.4846],[140,-0.4902],[142,-0.4902],[143,-0.4942],[146,-0.4942],[147,-0.4974],[148,-0.4974],[149,-0.499],[157,-0.499],[158,-0.5031],[159,-0.5023],[160,-0.5066],[161,-0.5065],[163,-0.5108],[166,-0.5108],[167,-0.5065],[169,-0.5066],[170,-0.5023],[171,-0.5023],[173,-0.4937],[174,-0.4937],[176,-0.4882],[179,-0.4876],[183,-0.4866],[190,-0.4962],[191,-0.499],[192,-0.499],[194,-0.5062],[195,-0.5062],[198,-0.517],[199,-0.517],[206,-0.5422],[207,-0.5435],[208,-0.5483],[210,-0.5531],[212,-0.5602],[213,-0.5602],[214,-0.5578],[215,-0.5531],[216,-0.5459],[217,-0.534],[218,-0.5198],[219,-0.5007],[220,-0.4793],[221,-0.4532],[222,-0.4223],[223,-0.3923],[225,-0.32],[227,-0.2333],[228,-0.1971],[230,-0.1104],[231,-0.0742],[233,-0.0164],[235,0.027],[236,0.0342],[237,0.0487],[238,0.0559],[239,0.0613],[240,0.0699],[245,0.1056],[246,0.1113],[249,0.1241],[254,0.1384],[255,0.1381],[256,0.1424],[260,0.1424],[261,0.1467],[263,0.1467],[264,0.1509],[267,0.1509],[268,0.1552],[269,0.1552],[270,0.1595],[271,0.1578],[272,0.1618],[274,0.1618],[275,0.1658],[279,0.1658],[280,0.1698],[296,0.1698],[305,0.1638],[323,-0.7327],[324,-0.7471],[325,-0.7507],[326,-0.7579],[330,-0.7723],[333,-0.7723],[334,-0.7759],[335,-0.7759],[336,-0.7723],[338,-0.7723],[339,-0.7687],[340,-0.7702],[343,-0.7702],[344,-0.7676],[347,-0.7676],[348,-0.7702],[351,-0.7702],[352,-0.7728],[354,-0.7728],[355,-0.7753],[356,-0.7745],[358,-0.7779],[360,-0.7779],[361,-0.7813],[364,-0.7813],[366,-0.7848],[371,-0.7848],[374,-0.7862],[379,-0.779],[380,-0.779],[388,-0.7623],[389,-0.7623],[390,-0.7579],[391,-0.7579],[392,-0.7535],[393,-0.7535],[394,-0.7491],[395,-0.7491],[396,-0.7447],[397,-0.7447],[398,-0.7403],[399,-0.7403],[400,-0.7359],[402,-0.7183],[404,-0.6925],[406,-0.6523],[407,-0.6255],[409,-0.5652],[413,-0.4312],[414,-0.401],[416,-0.3474],[419,-0.277],[420,-0.2609],[421,-0.2405],[423,-0.2134],[425,-0.1998],[426,-0.1862],[427,-0.1794],[428,-0.1794],[429,-0.1726],[433,-0.1726],[435,-0.1862],[436,-0.1916],[437,-0.2024],[438,-0.2149],[439,-0.2293],[441,-0.2651],[443,-0.2992],[444,-0.3171],[445,-0.3332],[448,-0.3763],[450,-0.3996],[451,-0.4103],[452,-0.4175],[454,-0.4399],[461,-0.479],[462,-0.479],[463,-0.4846],[464,-0.4846],[465,-0.4902],[467,-0.4902],[468,-0.4942],[471,-0.4942],[472,-0.4974],[473,-0.4974],[474,-0.499],[482,-0.499],[483,-0.5031],[484,-0.5023],[485,-0.5066],[486,-0.5065],[488,-0.5108],[491,-0.5108],[492,-0.5065],[494,-0.5066],[495,-0.5023],[496,-0.5023],[498,-0.4937],[499,-0.4937],[501,-0.4882],[504,-0.4876],[508,-0.4866],[515,-0.4962],[516,-0.499],[517,-0.499],[519,-0.5062],[520,-0.5062],[523,-0.517],[524,-0.517],[531,-0.5422],[532,-0.5435],[533,-0.5483],[535,-0.5531],[537,-0.5602],[538,-0.5602],[539,-0.5578],[540,-0.5531],[541,-0.5459],[542,-0.534],[543,-0.5198],[544,-0.5007],[545,-0.4793],[546,-0.4532],[547,-0.4223],[548,-0.3923],[550,-0.32],[552,-0.2333],[553,-0.1971],[555,-0.1104],[556,-0.0742],[558,-0.0164],[560,0.027],[561,0.0342],[562,0.0487],[563,0.0559],[564,0.0613],[565,0.0699],[570,0.1056],[571,0.1113],[574,0.1241],[579,0.1384],[580,0.1381],[581,0.1424],[585,0.1424],[586,0.1467],[588,0.1467],[589,0.1509],[592,0.1509],[593,0.1552],[594,0.1552],[595,0.1595],[596,0.1578],[597,0.1618],[599,0.1618],[600,0.1658],[604,0.1658],[605,0.1698],[621,0.1698],[630,0.1638],[692,2.7483],[713,2.739],[742,2.7547],[748,2.7679],[756,2.7723],[759,2.781],[761,2.7986],[763,2.8246],[765,2.8657],[776,3.7371],[802,3.6807],[807,3.6747],[809,3.6687],[825,3.2655],[842,2.7842],[887,0.1638],[888,0.1638],[889,0.1578],[893,0.1578],[894,0.1518],[895,0.1518],[929,-0.7327]]
       }
       lowerarm_r_fk_ctrl.Rotation.Y {
        type=MovieSceneScriptingFloatChannel
        default=0
        thinned_from=712
        keys [[-2,-3.3401],[630,-3.3428],[692,-1.9857],[765,-1.9844],[776,-1.9044],[782,-1.9603],[786,-1.9829],[809,-1.9885],[842,-1.9858],[887,-3.3428],[929,-3.3401]]
       }
       lowerarm_r_fk_ctrl.Rotation.Z {
        type=MovieSceneScriptingFloatChannel
        default=0
        thinned_from=654
        keys [[-2,76.2805],[630,76.2978],[646,73.6741],[660,29.27],[669,18.2869],[675,16.4163],[689,20.6656],[708,18.7998],[725,18.9782],[765,19.0263],[776,27.807],[782,30.783],[791,31.1128],[793,30.7377],[795,30.1629],[798,29.0369],[803,26.7613],[809,23.8063],[825,19.0122],[848,19.0029],[863,22.5907],[887,76.2978],[929,76.2805]]
       }
       lowerarm_r_fk_ctrl.Scale.X {type=MovieSceneScriptingFloatChannel, default=1, const [1,712]}
       lowerarm_r_fk_ctrl.Scale.Y {type=MovieSceneScriptingFloatChannel, default=1, const [1,712]}
       lowerarm_r_fk_ctrl.Scale.Z {type=MovieSceneScriptingFloatChannel, default=1, const [1,712]}
       hand_r_fk_ctrl.Location.X {type=MovieSceneScriptingFloatChannel, default=0, thinned_from=711, keys [[-2,0],[929,0]]}
       hand_r_fk_ctrl.Location.Y {type=MovieSceneScriptingFloatChannel, default=0, thinned_from=711, keys [[-2,0],[929,0]]}
       hand_r_fk_ctrl.Location.Z {type=MovieSceneScriptingFloatChannel, default=0, thinned_from=711, keys [[-2,0],[929,0]]}
       hand_r_fk_ctrl.Rotation.X {
        type=MovieSceneScriptingFloatChannel
        default=0
        thinned_from=711
        keys [[-2,-34.4342],[223,-34.7901],[296,-35.3716],[407,-34.5452],[549,-34.8266],[621,-35.3716],[638,-35.3609],[666,-19.9338],[692,-9.1316],[730,-9.1046],[736,-9.1182],[748,-9.1595],[759,-9.1807],[761,-9.2027],[763,-9.2561],[765,-9.3752],[776,-11.7666],[778,-11.4319],[780,-11.1463],[782,-10.9123],[784,-10.7328],[786,-10.6104],[788,-10.5479],[809,-10.5127],[842,-9.1832],[868,-11.4462],[888,-35.3612],[895,-35.3527],[929,-34.4342]]
       }
       hand_r_fk_ctrl.Rotation.Y {
        type=MovieSceneScriptingFloatChannel
        default=0
        thinned_from=647
        keys [[-2,19.5362],[296,19.3413],[638,19.3432],[660,28.4518],[663,28.0093],[674,19.2315],[689,25.5392],[707,17.3679],[741,20.8514],[763,20.8275],[765,20.7535],[809,20.6538],[842,20.8448],[863,7.0068],[887,19.3437],[895,19.3453],[929,19.5362]]
       }
       hand_r_fk_ctrl.Rotation.Z {
        type=MovieSceneScriptingFloatChannel
        default=0
        thinned_from=713
        keys [[-2,-11.1755],[224,-11.3066],[296,-11.4896],[407,-11.2124],[549,-11.3066],[621,-11.4896],[638,-11.4863],[653,13.6968],[666,0.2262],[682,-12.0817],[761,-12.0991],[763,-12.1121],[776,-12.3971],[809,-12.4024],[842,-12.0945],[865,10.7018],[867,9.0254],[888,-11.4861],[895,-11.4834],[929,-11.1755]]
       }
       hand_r_fk_ctrl.Scale.X {type=MovieSceneScriptingFloatChannel, default=1, const [1,711]}
       hand_r_fk_ctrl.Scale.Y {type=MovieSceneScriptingFloatChannel, default=1, const [1,711]}
       hand_r_fk_ctrl.Scale.Z {type=MovieSceneScriptingFloatChannel, default=1, const [1,711]}
       thumb_01_r_ctrl.Location.X {type=MovieSceneScriptingFloatChannel, default=1.5173, thinned_from=629, keys [[-2,1.5173],[912,1.5173]]}
       thumb_01_r_ctrl.Location.Y {type=MovieSceneScriptingFloatChannel, default=-0.0715, thinned_from=629, keys [[-2,-0.0715],[912,-0.0715]]}
       thumb_01_r_ctrl.Location.Z {type=MovieSceneScriptingFloatChannel, default=0.0001, thinned_from=629, keys [[-2,0.0001],[912,0.0001]]}
       thumb_01_r_ctrl.Rotation.X {
        type=MovieSceneScriptingFloatChannel
        default=-33.464
        thinned_from=631
        keys [[-2,-33.464],[634,-33.464],[644,-32.7025],[654,-35.1757],[668,-28.6536],[843,-28.6536],[853,-18.0232],[912,-33.464]]
       }
       thumb_01_r_ctrl.Rotation.Y {
        type=MovieSceneScriptingFloatChannel
        default=-13.0752
        thinned_from=631
        keys [[-2,-13.0752],[634,-13.0752],[644,20.4297],[654,35.7363],[668,9.3998],[843,9.3998],[853,23.192],[912,-13.0752]]
       }
       thumb_01_r_ctrl.Rotation.Z {
        type=MovieSceneScriptingFloatChannel
        default=-9.394
        thinned_from=631
        keys [[-2,-9.394],[634,-9.394],[645,-7.5564],[654,-6.9555],[668,-23.269],[843,-23.269],[853,-15.1704],[912,-9.394]]
       }
       thumb_01_r_ctrl.Scale.X {type=MovieSceneScriptingFloatChannel, default=1, const [1,629]}
       thumb_01_r_ctrl.Scale.Y {type=MovieSceneScriptingFloatChannel, default=1, const [1,629]}
       thumb_01_r_ctrl.Scale.Z {type=MovieSceneScriptingFloatChannel, default=1, const [1,629]}
       thumb_02_r_ctrl.Location.X {type=MovieSceneScriptingFloatChannel, default=0, thinned_from=630, keys [[-2,0],[912,0]]}
       thumb_02_r_ctrl.Location.Y {type=MovieSceneScriptingFloatChannel, default=0, thinned_from=630, keys [[-2,0],[912,0]]}
       thumb_02_r_ctrl.Location.Z {type=MovieSceneScriptingFloatChannel, default=0, thinned_from=630, keys [[-2,0],[912,0]]}
       thumb_02_r_ctrl.Rotation.X {
        type=MovieSceneScriptingFloatChannel
        default=1.2228
        thinned_from=631
        keys [[-2,1.2228],[634,1.2228],[644,0.1123],[654,0.116],[668,1.2228],[843,1.2228],[860,0.0996],[912,1.2228]]
       }
       thumb_02_r_ctrl.Rotation.Y {
        type=MovieSceneScriptingFloatChannel
        default=-2.3382
        thinned_from=632
        keys [[-2,-2.3382],[634,-2.3382],[644,-0.0065],[649,0],[654,-0.0055],[668,-2.3382],[843,-2.3382],[860,-2.6366],[912,-2.3382]]
       }
       thumb_02_r_ctrl.Rotation.Z {
        type=MovieSceneScriptingFloatChannel
        default=-13.4617
        thinned_from=631
        keys [[-2,-13.4617],[634,-13.4617],[644,11.7168],[654,11.6337],[668,-13.4617],[843,-13.4617],[866,24.8977],[912,-13.4617]]
       }
       thumb_02_r_ctrl.Scale.X {type=MovieSceneScriptingFloatChannel, default=1, const [1,630]}
       thumb_02_r_ctrl.Scale.Y {type=MovieSceneScriptingFloatChannel, default=1, const [1,630]}
       thumb_02_r_ctrl.Scale.Z {type=MovieSceneScriptingFloatChannel, default=1, const [1,630]}
       thumb_03_r_ctrl.Location.X {type=MovieSceneScriptingFloatChannel, default=0, thinned_from=630, keys [[-2,0],[912,0]]}
       thumb_03_r_ctrl.Location.Y {type=MovieSceneScriptingFloatChannel, default=0, thinned_from=630, keys [[-2,0],[912,0]]}
       thumb_03_r_ctrl.Location.Z {type=MovieSceneScriptingFloatChannel, default=0, thinned_from=630, keys [[-2,0],[912,0]]}
       thumb_03_r_ctrl.Rotation.X {
        type=MovieSceneScriptingFloatChannel
        default=2.0016
        thinned_from=631
        keys [[-2,2.0016],[634,2.0016],[644,1.8871],[654,1.0049],[668,2.0016],[843,2.0016],[860,1.5931],[912,2.0016]]
       }
       thumb_03_r_ctrl.Rotation.Y {
        type=MovieSceneScriptingFloatChannel
        default=0.1771
        thinned_from=631
        keys [[-2,0.1771],[634,0.1771],[644,-0.6904],[654,-1.7395],[668,0.1771],[843,0.1771],[860,-1.2248],[912,0.1771]]
       }
       thumb_03_r_ctrl.Rotation.Z {
        type=MovieSceneScriptingFloatChannel
        default=-52.3193
        thinned_from=631
        keys [[-2,-52.3193],[634,-52.3193],[644,-27.1777],[654,12.7105],[668,-52.3193],[843,-52.3193],[860,-9.7201],[912,-52.3193]]
       }
       thumb_03_r_ctrl.Scale.X {type=MovieSceneScriptingFloatChannel, default=1, const [1,630]}
       thumb_03_r_ctrl.Scale.Y {type=MovieSceneScriptingFloatChannel, default=1, const [1,630]}
       thumb_03_r_ctrl.Scale.Z {type=MovieSceneScriptingFloatChannel, default=1, const [1,630]}
       index_metacarpal_r_ctrl.Location.X {type=MovieSceneScriptingFloatChannel, default=0, thinned_from=635, keys [[-2,0],[929,0]]}
       index_metacarpal_r_ctrl.Location.Y {type=MovieSceneScriptingFloatChannel, default=0, thinned_from=635, keys [[-2,0],[929,0]]}
       index_metacarpal_r_ctrl.Location.Z {type=MovieSceneScriptingFloatChannel, default=0, thinned_from=635, keys [[-2,0],[929,0]]}
       index_metacarpal_r_ctrl.Rotation.X {type=MovieSceneScriptingFloatChannel, default=0, thinned_from=635, keys [[-2,0],[634,0],[692,0.4924],[897,0],[929,0]]}
       index_metacarpal_r_ctrl.Rotation.Y {type=MovieSceneScriptingFloatChannel, default=4.1035, thinned_from=635, keys [[-2,4.1035],[634,4.1035],[692,4.0739],[929,4.1035]]}
       index_metacarpal_r_ctrl.Rotation.Z {type=MovieSceneScriptingFloatChannel, default=0, thinned_from=635, keys [[-2,0],[634,0],[692,6.8983],[897,0],[929,0]]}
       index_metacarpal_r_ctrl.Scale.X {type=MovieSceneScriptingFloatChannel, default=1, const [1,635]}
       index_metacarpal_r_ctrl.Scale.Y {type=MovieSceneScriptingFloatChannel, default=1, const [1,635]}
       index_metacarpal_r_ctrl.Scale.Z {type=MovieSceneScriptingFloatChannel, default=1, const [1,635]}
       index_01_r_ctrl.Location.X {type=MovieSceneScriptingFloatChannel, default=0, thinned_from=630, keys [[-2,0],[912,0]]}
       index_01_r_ctrl.Location.Y {type=MovieSceneScriptingFloatChannel, default=0, thinned_from=630, keys [[-2,0],[912,0]]}
       index_01_r_ctrl.Location.Z {type=MovieSceneScriptingFloatChannel, default=0, thinned_from=630, keys [[-2,0],[912,0]]}
       index_01_r_ctrl.Rotation.X {
        type=MovieSceneScriptingFloatChannel
        default=7.7952
        thinned_from=631
        keys [[-2,7.7952],[634,7.7952],[644,5.0113],[654,3.8143],[668,12.0544],[843,12.0544],[874,5.9452],[912,7.7952]]
       }
       index_01_r_ctrl.Rotation.Y {
        type=MovieSceneScriptingFloatChannel
        default=-9.2432
        thinned_from=632
        keys [[-2,-9.2432],[634,-9.2432],[644,-10.009],[654,-10.4969],[668,0.609],[843,0.609],[874,-23.6548],[888,-17.5339],[912,-9.2432]]
       }
       index_01_r_ctrl.Rotation.Z {
        type=MovieSceneScriptingFloatChannel
        default=-26.0805
        thinned_from=631
        keys [[-2,-26.0805],[634,-26.0805],[644,-13.106],[654,-6.4624],[668,-78.4897],[843,-78.4897],[862,2.3332],[912,-26.0805]]
       }
       index_01_r_ctrl.Scale.X {type=MovieSceneScriptingFloatChannel, default=1, const [1,630]}
       index_01_r_ctrl.Scale.Y {type=MovieSceneScriptingFloatChannel, default=1, const [1,630]}
       index_01_r_ctrl.Scale.Z {type=MovieSceneScriptingFloatChannel, default=1, const [1,630]}
       index_02_r_ctrl.Location.X {type=MovieSceneScriptingFloatChannel, default=0.3298, thinned_from=630, keys [[-2,0.3298],[912,0.3298]]}
       index_02_r_ctrl.Location.Y {type=MovieSceneScriptingFloatChannel, default=-0.1751, thinned_from=630, keys [[-2,-0.1751],[912,-0.1751]]}
       index_02_r_ctrl.Location.Z {type=MovieSceneScriptingFloatChannel, default=-0.0001, thinned_from=630, keys [[-2,-0.0001],[912,-0.0001]]}
       index_02_r_ctrl.Rotation.X {
        type=MovieSceneScriptingFloatChannel
        default=-0.379
        thinned_from=631
        keys [[-2,-0.379],[634,-0.379],[644,-0.3622],[654,-0.3597],[668,-0.1496],[843,-0.1496],[874,-0.3912],[912,-0.379]]
       }
       index_02_r_ctrl.Rotation.Y {
        type=MovieSceneScriptingFloatChannel
        default=-0.0975
        thinned_from=631
        keys [[-2,-0.0975],[634,-0.0975],[644,-0.006],[654,0.0355],[668,-0.3617],[843,-0.3617],[874,-0.0109],[912,-0.0975]]
       }
       index_02_r_ctrl.Rotation.Z {
        type=MovieSceneScriptingFloatChannel
        default=0
        thinned_from=631
        keys [[-2,0],[634,0],[644,12.7165],[654,19.2694],[668,-53.1044],[843,-53.1044],[865,17.9611],[912,0]]
       }
       index_02_r_ctrl.Scale.X {type=MovieSceneScriptingFloatChannel, default=1, const [1,630]}
       index_02_r_ctrl.Scale.Y {type=MovieSceneScriptingFloatChannel, default=1, const [1,630]}
       index_02_r_ctrl.Scale.Z {type=MovieSceneScriptingFloatChannel, default=1, const [1,630]}
       index_03_r_ctrl.Location.X {type=MovieSceneScriptingFloatChannel, default=0.2932, thinned_from=630, keys [[-2,0.2932],[912,0.2932]]}
       index_03_r_ctrl.Location.Y {type=MovieSceneScriptingFloatChannel, default=0.1198, thinned_from=630, keys [[-2,0.1198],[912,0.1198]]}
       index_03_r_ctrl.Location.Z {type=MovieSceneScriptingFloatChannel, default=0.005, thinned_from=630, keys [[-2,0.005],[912,0.005]]}
       index_03_r_ctrl.Rotation.X {
        type=MovieSceneScriptingFloatChannel
        default=0.0155
        thinned_from=631
        keys [[-2,0.0155],[634,0.0155],[644,0.1851],[654,0.022],[668,-0.6249],[843,-0.6249],[874,-0.0977],[912,0.0155]]
       }
       index_03_r_ctrl.Rotation.Y {
        type=MovieSceneScriptingFloatChannel
        default=0.793
        thinned_from=631
        keys [[-2,0.793],[634,0.793],[644,0.7103],[654,0.7321],[668,0.4885],[843,0.4885],[874,0.7871],[912,0.793]]
       }
       index_03_r_ctrl.Rotation.Z {
        type=MovieSceneScriptingFloatChannel
        default=33.9964
        thinned_from=631
        keys [[-2,33.9964],[634,33.9964],[644,46.7146],[654,33.7946],[668,-19.111],[843,-19.111],[874,25.7955],[912,33.9964]]
       }
       index_03_r_ctrl.Scale.X {type=MovieSceneScriptingFloatChannel, default=1, const [1,630]}
       index_03_r_ctrl.Scale.Y {type=MovieSceneScriptingFloatChannel, default=1, const [1,630]}
       index_03_r_ctrl.Scale.Z {type=MovieSceneScriptingFloatChannel, default=1, const [1,630]}
       middle_metacarpal_r_ctrl.Location.X {type=MovieSceneScriptingFloatChannel, default=0, thinned_from=635, keys [[-2,0],[929,0]]}
       middle_metacarpal_r_ctrl.Location.Y {type=MovieSceneScriptingFloatChannel, default=0, thinned_from=635, keys [[-2,0],[929,0]]}
       middle_metacarpal_r_ctrl.Location.Z {type=MovieSceneScriptingFloatChannel, default=0, thinned_from=635, keys [[-2,0],[929,0]]}
       middle_metacarpal_r_ctrl.Rotation.X {type=MovieSceneScriptingFloatChannel, default=0, thinned_from=635, keys [[-2,0],[634,0],[692,0.0543],[897,0],[929,0]]}
       middle_metacarpal_r_ctrl.Rotation.Y {type=MovieSceneScriptingFloatChannel, default=1.2252, thinned_from=635, keys [[-2,1.2252],[929,1.2252]]}
       middle_metacarpal_r_ctrl.Rotation.Z {type=MovieSceneScriptingFloatChannel, default=0, thinned_from=635, keys [[-2,0],[634,0],[692,2.5422],[897,0],[929,0]]}
       middle_metacarpal_r_ctrl.Scale.X {type=MovieSceneScriptingFloatChannel, default=1, const [1,635]}
       middle_metacarpal_r_ctrl.Scale.Y {type=MovieSceneScriptingFloatChannel, default=1, const [1,635]}
       middle_metacarpal_r_ctrl.Scale.Z {type=MovieSceneScriptingFloatChannel, default=1, const [1,635]}
       middle_01_r_ctrl.Location.X {type=MovieSceneScriptingFloatChannel, default=0, thinned_from=630, keys [[-2,0],[912,0]]}
       middle_01_r_ctrl.Location.Y {type=MovieSceneScriptingFloatChannel, default=0, thinned_from=630, keys [[-2,0],[912,0]]}
       middle_01_r_ctrl.Location.Z {type=MovieSceneScriptingFloatChannel, default=0, thinned_from=630, keys [[-2,0],[912,0]]}
       middle_01_r_ctrl.Rotation.X {
        type=MovieSceneScriptingFloatChannel
        default=3.7685
        thinned_from=631
        keys [[-2,3.7685],[634,3.7685],[644,2.5122],[654,2.5221],[668,5.5207],[843,5.5207],[874,2.7954],[912,3.7685]]
       }
       middle_01_r_ctrl.Rotation.Y {
        type=MovieSceneScriptingFloatChannel
        default=-4.0758
        thinned_from=631
        keys [[-2,-4.0758],[634,-4.0758],[644,-4.4822],[654,-4.4656],[668,0.5592],[843,0.5592],[874,-9.1687],[912,-4.0758]]
       }
       middle_01_r_ctrl.Rotation.Z {
        type=MovieSceneScriptingFloatChannel
        default=-28.6993
        thinned_from=631
        keys [[-2,-28.6993],[634,-28.6993],[644,-15.9339],[654,-16.1507],[668,-81.6428],[843,-81.6428],[865,-8.9989],[912,-28.6993]]
       }
       middle_01_r_ctrl.Scale.X {type=MovieSceneScriptingFloatChannel, default=1, const [1,630]}
       middle_01_r_ctrl.Scale.Y {type=MovieSceneScriptingFloatChannel, default=1, const [1,630]}
       middle_01_r_ctrl.Scale.Z {type=MovieSceneScriptingFloatChannel, default=1, const [1,630]}
       middle_02_r_ctrl.Location.X {type=MovieSceneScriptingFloatChannel, default=0.0008, thinned_from=630, keys [[-2,0.0008],[912,0.0008]]}
       middle_02_r_ctrl.Location.Y {type=MovieSceneScriptingFloatChannel, default=0.0009, thinned_from=630, keys [[-2,0.0009],[912,0.0009]]}
       middle_02_r_ctrl.Location.Z {type=MovieSceneScriptingFloatChannel, default=0.0476, thinned_from=630, keys [[-2,0.0476],[912,0.0476]]}
       middle_02_r_ctrl.Rotation.X {
        type=MovieSceneScriptingFloatChannel
        default=-3.8525
        thinned_from=631
        keys [[-2,-3.8525],[634,-3.8525],[644,-3.5494],[654,-3.5431],[668,-2.0122],[843,-2.0122],[874,-3.8405],[912,-3.8525]]
       }
       middle_02_r_ctrl.Rotation.Y {
        type=MovieSceneScriptingFloatChannel
        default=-0.3785
        thinned_from=631
        keys [[-2,-0.3785],[634,-0.3785],[644,0.4897],[654,0.4772],[668,-3.3077],[843,-3.3077],[874,0.4855],[912,-0.3785]]
       }
       middle_02_r_ctrl.Rotation.Z {
        type=MovieSceneScriptingFloatChannel
        default=0
        thinned_from=631
        keys [[-2,0],[634,0],[644,12.6805],[654,12.4636],[668,-53.0592],[843,-53.0592],[873,4.2753],[912,0]]
       }
       middle_02_r_ctrl.Scale.X {type=MovieSceneScriptingFloatChannel, default=1, const [1,630]}
       middle_02_r_ctrl.Scale.Y {type=MovieSceneScriptingFloatChannel, default=1, const [1,630]}
       middle_02_r_ctrl.Scale.Z {type=MovieSceneScriptingFloatChannel, default=1, const [1,630]}
       middle_03_r_ctrl.Location.X {type=MovieSceneScriptingFloatChannel, default=0, thinned_from=630, keys [[-2,0],[912,0]]}
       middle_03_r_ctrl.Location.Y {type=MovieSceneScriptingFloatChannel, default=0, thinned_from=630, keys [[-2,0],[912,0]]}
       middle_03_r_ctrl.Location.Z {type=MovieSceneScriptingFloatChannel, default=0, thinned_from=630, keys [[-2,0],[912,0]]}
       middle_03_r_ctrl.Rotation.X {
        type=MovieSceneScriptingFloatChannel
        default=1.5318
        thinned_from=631
        keys [[-2,1.5318],[634,1.5318],[644,1.7634],[654,1.2873],[668,-0.5082],[843,-0.5082],[874,1.8896],[912,1.5318]]
       }
       middle_03_r_ctrl.Rotation.Y {
        type=MovieSceneScriptingFloatChannel
        default=1.7844
        thinned_from=631
        keys [[-2,1.7844],[634,1.7844],[644,1.2756],[654,1.7499],[668,2.296],[843,2.296],[874,1.3999],[912,1.7844]]
       }
       middle_03_r_ctrl.Rotation.Z {
        type=MovieSceneScriptingFloatChannel
        default=22.205
        thinned_from=631
        keys [[-2,22.205],[634,22.205],[644,34.916],[654,17.1043],[668,-30.9336],[843,-30.9336],[874,35.0301],[912,22.205]]
       }
       middle_03_r_ctrl.Scale.X {type=MovieSceneScriptingFloatChannel, default=1, const [1,630]}
       middle_03_r_ctrl.Scale.Y {type=MovieSceneScriptingFloatChannel, default=1, const [1,630]}
       middle_03_r_ctrl.Scale.Z {type=MovieSceneScriptingFloatChannel, default=1, const [1,630]}
       ring_metacarpal_r_ctrl.Location.X {type=MovieSceneScriptingFloatChannel, default=0, thinned_from=635, keys [[-2,0],[929,0]]}
       ring_metacarpal_r_ctrl.Location.Y {type=MovieSceneScriptingFloatChannel, default=0, thinned_from=635, keys [[-2,0],[929,0]]}
       ring_metacarpal_r_ctrl.Location.Z {type=MovieSceneScriptingFloatChannel, default=0, thinned_from=635, keys [[-2,0],[929,0]]}
       ring_metacarpal_r_ctrl.Rotation.X {type=MovieSceneScriptingFloatChannel, default=0, thinned_from=635, keys [[-2,0],[634,0],[692,0.0178],[905,0],[929,0]]}
       ring_metacarpal_r_ctrl.Rotation.Y {
        type=MovieSceneScriptingFloatChannel
        default=-4.4979
        thinned_from=635
        keys [[-2,-4.4979],[634,-4.4979],[692,-2.1591],[897,-4.4979],[929,-4.4979]]
       }
       ring_metacarpal_r_ctrl.Rotation.Z {type=MovieSceneScriptingFloatChannel, default=0, thinned_from=635, keys [[-2,0],[634,0],[692,-0.4736],[897,0],[929,0]]}
       ring_metacarpal_r_ctrl.Scale.X {type=MovieSceneScriptingFloatChannel, default=1, const [1,635]}
       ring_metacarpal_r_ctrl.Scale.Y {type=MovieSceneScriptingFloatChannel, default=1, const [1,635]}
       ring_metacarpal_r_ctrl.Scale.Z {type=MovieSceneScriptingFloatChannel, default=1, const [1,635]}
       ring_01_r_ctrl.Location.X {type=MovieSceneScriptingFloatChannel, default=0, thinned_from=630, keys [[-2,0],[912,0]]}
       ring_01_r_ctrl.Location.Y {type=MovieSceneScriptingFloatChannel, default=0, thinned_from=630, keys [[-2,0],[912,0]]}
       ring_01_r_ctrl.Location.Z {type=MovieSceneScriptingFloatChannel, default=0, thinned_from=630, keys [[-2,0],[912,0]]}
       ring_01_r_ctrl.Rotation.X {
        type=MovieSceneScriptingFloatChannel
        default=-3.1309
        thinned_from=631
        keys [[-2,-3.1309],[634,-3.1309],[644,-1.8031],[654,-1.8157],[668,-5.6306],[843,-5.6306],[874,-2.0721],[912,-3.1309]]
       }
       ring_01_r_ctrl.Rotation.Y {
        type=MovieSceneScriptingFloatChannel
        default=4.6931
        thinned_from=632
        keys [[-2,4.6931],[634,4.6931],[644,4.9],[654,4.8849],[668,0.3201],[843,0.3201],[872,2.4359],[884,4.141],[912,4.6931]]
       }
       ring_01_r_ctrl.Rotation.Z {
        type=MovieSceneScriptingFloatChannel
        default=-25.3784
        thinned_from=631
        keys [[-2,-25.3784],[634,-25.3784],[644,-12.5932],[654,-12.8103],[668,-78.3704],[843,-78.3704],[867,-4.0287],[912,-25.3784]]
       }
       ring_01_r_ctrl.Scale.X {type=MovieSceneScriptingFloatChannel, default=1, const [1,630]}
       ring_01_r_ctrl.Scale.Y {type=MovieSceneScriptingFloatChannel, default=1, const [1,630]}
       ring_01_r_ctrl.Scale.Z {type=MovieSceneScriptingFloatChannel, default=1, const [1,630]}
       ring_02_r_ctrl.Location.X {type=MovieSceneScriptingFloatChannel, default=0, const [0,630]}
       ring_02_r_ctrl.Location.Y {type=MovieSceneScriptingFloatChannel, default=0, const [0,630]}
       ring_02_r_ctrl.Location.Z {type=MovieSceneScriptingFloatChannel, default=0, const [0,630]}
       ring_02_r_ctrl.Rotation.X {type=MovieSceneScriptingFloatChannel, default=0, thinned_from=630, keys [[-2,0],[912,0]]}
       ring_02_r_ctrl.Rotation.Y {type=MovieSceneScriptingFloatChannel, default=0, thinned_from=630, keys [[-2,0],[912,0]]}
       ring_02_r_ctrl.Rotation.Z {
        type=MovieSceneScriptingFloatChannel
        default=0
        thinned_from=631
        keys [[-2,0],[634,0],[644,12.7169],[654,12.4997],[668,-53.1046],[843,-53.1046],[880,-5.0001],[912,0]]
       }
       ring_02_r_ctrl.Scale.X {type=MovieSceneScriptingFloatChannel, default=1, const [1,630]}
       ring_02_r_ctrl.Scale.Y {type=MovieSceneScriptingFloatChannel, default=1, const [1,630]}
       ring_02_r_ctrl.Scale.Z {type=MovieSceneScriptingFloatChannel, default=1, const [1,630]}
       ring_03_r_ctrl.Location.X {type=MovieSceneScriptingFloatChannel, default=0, thinned_from=630, keys [[-2,0],[912,0]]}
       ring_03_r_ctrl.Location.Y {type=MovieSceneScriptingFloatChannel, default=0, thinned_from=630, keys [[-2,0],[912,0]]}
       ring_03_r_ctrl.Location.Z {type=MovieSceneScriptingFloatChannel, default=0, thinned_from=630, keys [[-2,0],[912,0]]}
       ring_03_r_ctrl.Rotation.X {type=MovieSceneScriptingFloatChannel, default=0, thinned_from=630, keys [[-2,0],[912,0]]}
       ring_03_r_ctrl.Rotation.Y {type=MovieSceneScriptingFloatChannel, default=0, thinned_from=630, keys [[-2,0],[912,0]]}
       ring_03_r_ctrl.Rotation.Z {
        type=MovieSceneScriptingFloatChannel
        default=22.1957
        thinned_from=631
        keys [[-2,22.1957],[634,22.1957],[644,34.9125],[654,17.1009],[668,-30.9089],[843,-30.9089],[874,35.0216],[912,22.1957]]
       }
       ring_03_r_ctrl.Scale.X {type=MovieSceneScriptingFloatChannel, default=1, const [1,630]}
       ring_03_r_ctrl.Scale.Y {type=MovieSceneScriptingFloatChannel, default=1, const [1,630]}
       ring_03_r_ctrl.Scale.Z {type=MovieSceneScriptingFloatChannel, default=1, const [1,630]}
       pinky_metacarpal_r_ctrl.Location.X {type=MovieSceneScriptingFloatChannel, default=0, thinned_from=635, keys [[-2,0],[929,0]]}
       pinky_metacarpal_r_ctrl.Location.Y {type=MovieSceneScriptingFloatChannel, default=0, thinned_from=635, keys [[-2,0],[929,0]]}
       pinky_metacarpal_r_ctrl.Location.Z {type=MovieSceneScriptingFloatChannel, default=0, thinned_from=635, keys [[-2,0],[929,0]]}
       pinky_metacarpal_r_ctrl.Rotation.X {type=MovieSceneScriptingFloatChannel, default=-4.938, thinned_from=635, keys [[-2,-4.938],[634,-4.938],[692,-4.8921],[929,-4.938]]}
       pinky_metacarpal_r_ctrl.Rotation.Y {
        type=MovieSceneScriptingFloatChannel
        default=-4.2094
        thinned_from=635
        keys [[-2,-4.2094],[634,-4.2094],[692,-1.3617],[897,-4.2094],[929,-4.2094]]
       }
       pinky_metacarpal_r_ctrl.Rotation.Z {
        type=MovieSceneScriptingFloatChannel
        default=-2.0167
        thinned_from=635
        keys [[-2,-2.0167],[634,-2.0167],[692,-3.7589],[897,-2.0167],[929,-2.0167]]
       }
       pinky_metacarpal_r_ctrl.Scale.X {type=MovieSceneScriptingFloatChannel, default=1, const [1,635]}
       pinky_metacarpal_r_ctrl.Scale.Y {type=MovieSceneScriptingFloatChannel, default=1, const [1,635]}
       pinky_metacarpal_r_ctrl.Scale.Z {type=MovieSceneScriptingFloatChannel, default=1, const [1,635]}
       pinky_01_r_ctrl.Location.X {type=MovieSceneScriptingFloatChannel, default=0, thinned_from=630, keys [[-2,0],[912,0]]}
       pinky_01_r_ctrl.Location.Y {type=MovieSceneScriptingFloatChannel, default=0, thinned_from=630, keys [[-2,0],[912,0]]}
       pinky_01_r_ctrl.Location.Z {type=MovieSceneScriptingFloatChannel, default=0, thinned_from=630, keys [[-2,0],[912,0]]}
       pinky_01_r_ctrl.Rotation.X {
        type=MovieSceneScriptingFloatChannel
        default=-1.5748
        thinned_from=631
        keys [[-2,-1.5748],[634,-1.5748],[644,0.1519],[654,0.1292],[668,-6.7232],[843,-6.7232],[874,-2.3455],[912,-1.5748]]
       }
       pinky_01_r_ctrl.Rotation.Y {
        type=MovieSceneScriptingFloatChannel
        default=7.2218
        thinned_from=632
        keys [[-2,7.2218],[634,7.2218],[644,6.8375],[654,6.8235],[668,3.0764],[843,3.0764],[871,7.5251],[893,0.2329],[912,7.2218]]
       }
       pinky_01_r_ctrl.Rotation.Z {
        type=MovieSceneScriptingFloatChannel
        default=-7.7286
        thinned_from=631
        keys [[-2,-7.7286],[634,-7.7286],[644,-2.5122],[654,-2.7301],[668,-60.9146],[843,-60.9146],[861,6.9287],[912,-7.7286]]
       }
       pinky_01_r_ctrl.Scale.X {type=MovieSceneScriptingFloatChannel, default=1, const [1,630]}
       pinky_01_r_ctrl.Scale.Y {type=MovieSceneScriptingFloatChannel, default=1, const [1,630]}
       pinky_01_r_ctrl.Scale.Z {type=MovieSceneScriptingFloatChannel, default=1, const [1,630]}
       pinky_02_r_ctrl.Location.X {type=MovieSceneScriptingFloatChannel, default=0, thinned_from=630, keys [[-2,0],[912,0]]}
       pinky_02_r_ctrl.Location.Y {type=MovieSceneScriptingFloatChannel, default=0, thinned_from=630, keys [[-2,0],[912,0]]}
       pinky_02_r_ctrl.Location.Z {type=MovieSceneScriptingFloatChannel, default=0, thinned_from=630, keys [[-2,0],[912,0]]}
       pinky_02_r_ctrl.Rotation.X {
        type=MovieSceneScriptingFloatChannel
        default=-0.0384
        thinned_from=631
        keys [[-2,-0.0384],[634,-0.0384],[644,0.759],[654,0.7468],[668,-2.9617],[843,-2.9617],[874,0.779],[912,-0.0384]]
       }
       pinky_02_r_ctrl.Rotation.Y {
        type=MovieSceneScriptingFloatChannel
        default=3.6728
        thinned_from=631
        keys [[-2,3.6728],[634,3.6728],[644,3.3136],[654,3.3086],[668,2.1734],[843,2.1734],[874,3.5896],[912,3.6728]]
       }
       pinky_02_r_ctrl.Rotation.Z {
        type=MovieSceneScriptingFloatChannel
        default=0
        thinned_from=631
        keys [[-2,0],[634,0],[644,5.1045],[654,4.887],[668,-53.1596],[843,-53.1596],[870,-16.5974],[912,0]]
       }
       pinky_02_r_ctrl.Scale.X {type=MovieSceneScriptingFloatChannel, default=1, const [1,630]}
       pinky_02_r_ctrl.Scale.Y {type=MovieSceneScriptingFloatChannel, default=1, const [1,630]}
       pinky_02_r_ctrl.Scale.Z {type=MovieSceneScriptingFloatChannel, default=1, const [1,630]}
       pinky_03_r_ctrl.Location.X {type=MovieSceneScriptingFloatChannel, default=0, thinned_from=630, keys [[-2,0],[912,0]]}
       pinky_03_r_ctrl.Location.Y {type=MovieSceneScriptingFloatChannel, default=0, thinned_from=630, keys [[-2,0],[912,0]]}
       pinky_03_r_ctrl.Location.Z {type=MovieSceneScriptingFloatChannel, default=0, thinned_from=630, keys [[-2,0],[912,0]]}
       pinky_03_r_ctrl.Rotation.X {
        type=MovieSceneScriptingFloatChannel
        default=-0.1956
        thinned_from=631
        keys [[-2,-0.1956],[634,-0.1956],[644,0.0633],[654,0.0598],[668,-1.0045],[843,-1.0045],[874,0.0555],[912,-0.1956]]
       }
       pinky_03_r_ctrl.Rotation.Y {
        type=MovieSceneScriptingFloatChannel
        default=1.1092
        thinned_from=631
        keys [[-2,1.1092],[634,1.1092],[644,1.0405],[654,1.0385],[668,0.5095],[843,0.5095],[874,1.1249],[912,1.1092]]
       }
       pinky_03_r_ctrl.Rotation.Z {
        type=MovieSceneScriptingFloatChannel
        default=0
        thinned_from=631
        keys [[-2,0],[634,0],[644,5.0777],[654,4.8605],[668,-53.1072],[843,-53.1072],[874,12.8283],[912,0]]
       }
       pinky_03_r_ctrl.Scale.X {type=MovieSceneScriptingFloatChannel, default=1, const [1,630]}
       pinky_03_r_ctrl.Scale.Y {type=MovieSceneScriptingFloatChannel, default=1, const [1,630]}
       pinky_03_r_ctrl.Scale.Z {type=MovieSceneScriptingFloatChannel, default=1, const [1,630]}
       hand_r_ik_ctrl.Location.X {
        type=MovieSceneScriptingFloatChannel
        default=0
        thinned_from=751
        keys [[-2,-66.0294],[-1,-66.5896],[3,-66.9846],[10,-66.4258],[28,-63.7278],[37,-63.8588],[66,-66.2832],[93,-63.5821],[100,-63.804],[132,-70.0462],[141,-70.545],[201,-66.9846],[216,-67.1786],[221,-66.5486],[242,-60.3255],[249,-59.1609],[304,-60.2192],[309,-59.9567],[323,-66.0294],[324,-66.5896],[328,-66.9846],[335,-66.4258],[353,-63.7278],[362,-63.8588],[391,-66.2832],[418,-63.5821],[425,-63.804],[457,-70.0462],[466,-70.545],[526,-66.9846],[541,-67.1786],[546,-66.5486],[567,-60.3255],[574,-59.1609],[629,-60.2192],[634,-59.9567],[676,-66.0294],[677,-66.5896],[681,-66.9846],[688,-66.4258],[706,-63.7278],[715,-63.8588],[744,-66.2832],[755,-65.4688],[759,-64.9421],[770,-60.324],[803,-59.3299],[837,-66.0294],[895,-59.3299],[929,-66.0294]]
       }
       hand_r_ik_ctrl.Location.Y {
        type=MovieSceneScriptingFloatChannel
        default=0
        thinned_from=751
        keys [[-2,21.3873],[5,21.7868],[22,23.5476],[27,23.7913],[50,23.1845],[64,23.4751],[75,24.2716],[85,25.1929],[95,25.4655],[100,25.3429],[104,25.0234],[110,24.2351],[126,21.4627],[131,20.7733],[136,20.3315],[140,20.1644],[152,20.3054],[175,21.2027],[203,22.6349],[214,22.7649],[218,22.6235],[221,22.359],[223,22.0885],[225,21.728],[228,21.0207],[232,19.8322],[238,17.6809],[243,15.848],[245,15.1986],[247,14.6393],[249,14.174],[251,13.8205],[254,13.4506],[257,13.1637],[264,12.6695],[270,12.3997],[279,12.1541],[286,12.1024],[290,12.1509],[291,12.179],[293,12.2918],[295,12.4656],[297,12.6889],[300,13.1266],[304,13.83],[309,14.8056],[323,21.3873],[330,21.7868],[347,23.5476],[352,23.7913],[375,23.1845],[389,23.4751],[400,24.2716],[410,25.1929],[420,25.4655],[425,25.3429],[429,25.0234],[435,24.2351],[451,21.4627],[456,20.7733],[461,20.3315],[465,20.1644],[477,20.3054],[500,21.2027],[528,22.6349],[539,22.7649],[543,22.6235],[546,22.359],[548,22.0885],[550,21.728],[553,21.0207],[557,19.8322],[563,17.6809],[568,15.848],[570,15.1986],[572,14.6393],[574,14.174],[576,13.8205],[579,13.4506],[582,13.1637],[589,12.6695],[595,12.3997],[604,12.1541],[611,12.1024],[615,12.1509],[616,12.179],[618,12.2918],[620,12.4656],[622,12.6889],[625,13.1266],[629,13.83],[634,14.8056],[676,21.3873],[683,21.7868],[700,23.5476],[705,23.7913],[728,23.1845],[742,23.4751],[753,24.2716],[759,24.8608],[770,12.1072],[775,12.1509],[776,12.179],[778,12.2918],[780,12.4656],[782,12.6889],[785,13.1266],[789,13.83],[801,16.0954],[803,16.4325],[837,21.3873],[887,14.9975],[895,16.4325],[929,21.3873]]
       }
       hand_r_ik_ctrl.Location.Z {
        type=MovieSceneScriptingFloatChannel
        default=0
        thinned_from=751
        keys [[-2,40.3807],[19,37.6043],[28,36.9399],[38,37.171],[66,39.1594],[79,38.8121],[86,38.1587],[103,34.7527],[107,34.2859],[111,34.1382],[115,34.2636],[129,35.6747],[133,35.796],[138,35.5383],[166,31.5337],[181,30.2308],[194,30.1624],[203,30.6264],[212,31.7033],[216,32.4293],[235,38.0477],[239,38.894],[245,39.5504],[292,41.9326],[309,41.8916],[331,39.3155],[344,37.6043],[353,36.9399],[363,37.171],[391,39.1594],[404,38.8121],[411,38.1587],[428,34.7527],[432,34.2859],[436,34.1382],[440,34.2636],[454,35.6747],[458,35.796],[463,35.5383],[491,31.5337],[506,30.2308],[519,30.1624],[528,30.6264],[537,31.7033],[541,32.4293],[560,38.0477],[564,38.894],[570,39.5504],[617,41.9326],[634,41.8916],[676,40.3807],[697,37.6043],[706,36.9399],[716,37.171],[744,39.1594],[755,38.9394],[759,38.6626],[770,41.801],[803,41.7076],[837,40.3807],[887,41.8714],[895,41.7076],[929,40.3807]]
       }
       hand_r_ik_ctrl.Rotation.X {
        type=MovieSceneScriptingFloatChannel
        default=0
        thinned_from=751
        keys [[-2,77.891],[65,78.6556],[84,77.3027],[87,76.6079],[89,75.8089],[92,74.0836],[104,64.6241],[107,62.679],[110,61.1882],[113,60.1707],[116,59.5678],[119,59.3766],[125,59.9009],[133,61.133],[168,59.4496],[183,56.9758],[205,56.1507],[214,56.5269],[217,56.993],[224,59.1062],[228,60.7962],[234,64.3113],[240,68.5269],[247,73.88],[253,77.468],[263,82.1304],[281,86.1225],[293,87.0775],[309,85.3479],[323,77.891],[390,78.6556],[409,77.3027],[412,76.6079],[414,75.8089],[417,74.0836],[429,64.6241],[432,62.679],[435,61.1882],[438,60.1707],[441,59.5678],[444,59.3766],[450,59.9009],[458,61.133],[493,59.4496],[508,56.9758],[530,56.1507],[539,56.5269],[542,56.993],[549,59.1062],[553,60.7962],[559,64.3113],[565,68.5269],[572,73.88],[578,77.468],[588,82.1304],[606,86.1225],[618,87.0775],[634,85.3479],[676,77.891],[743,78.6556],[759,77.6202],[770,86.5883],[781,86.9849],[803,84.4392],[837,77.891],[887,85.2135],[895,84.4392],[929,77.891]]
       }
       hand_r_ik_ctrl.Rotation.Y {
        type=MovieSceneScriptingFloatChannel
        default=0
        thinned_from=751
        keys [[-2,-6.5201],[-1,-7.2762],[2,-8.1297],[7,-9.4459],[11,-10.4187],[21,-12.7937],[23,-13.2144],[25,-13.5654],[27,-13.8244],[29,-13.9856],[44,-14.3568],[52,-14.3445],[75,-13.5949],[78,-13.6028],[82,-13.7637],[85,-13.7773],[87,-13.8458],[88,-13.9366],[89,-14.0875],[90,-14.2948],[91,-14.5531],[92,-14.8467],[100,-17.6526],[102,-18.2628],[103,-18.5139],[106,-19.0757],[108,-19.3837],[111,-19.6941],[117,-20.0711],[121,-20.51],[128,-21.5848],[133,-22.167],[137,-22.4323],[160,-22.6158],[171,-23.1364],[174,-23.4231],[182,-24.698],[189,-25.362],[196,-25.6216],[201,-25.5591],[205,-25.2899],[208,-24.9128],[211,-24.3368],[213,-23.79],[214,-23.4481],[216,-22.6251],[217,-22.1364],[218,-21.5751],[219,-20.9496],[220,-20.2732],[221,-19.5248],[222,-18.701],[223,-17.8116],[224,-16.8591],[225,-15.8397],[226,-14.7584],[227,-13.6099],[228,-12.3962],[229,-11.1281],[230,-9.8026],[231,-8.423],[232,-6.9932],[233,-5.5258],[234,-4.0178],[235,-2.4813],[236,-0.9137],[237,0.6664],[238,2.2535],[239,3.8241],[240,5.3725],[241,6.8851],[242,8.3566],[243,9.7886],[244,11.1829],[245,12.5292],[246,13.8209],[247,15.0509],[248,16.2042],[249,17.2771],[250,18.2663],[251,19.1638],[252,19.9709],[253,20.7218],[254,21.4308],[256,22.7198],[258,23.8517],[260,24.8327],[262,25.6781],[264,26.3979],[266,26.9835],[270,27.8324],[274,28.4154],[281,28.9507],[290,29.2355],[292,29.205],[295,28.9499],[297,28.6243],[300,27.8909],[309,24.8804],[323,-6.5201],[324,-7.2762],[327,-8.1297],[332,-9.4459],[336,-10.4187],[346,-12.7937],[348,-13.2144],[350,-13.5654],[352,-13.8244],[354,-13.9856],[369,-14.3568],[377,-14.3445],[400,-13.5949],[403,-13.6028],[407,-13.7637],[410,-13.7773],[412,-13.8458],[413,-13.9366],[414,-14.0875],[415,-14.2948],[416,-14.5531],[417,-14.8467],[425,-17.6526],[427,-18.2628],[428,-18.5139],[431,-19.0757],[433,-19.3837],[436,-19.6941],[442,-20.0711],[446,-20.51],[453,-21.5848],[458,-22.167],[462,-22.4323],[485,-22.6158],[496,-23.1364],[499,-23.4231],[507,-24.698],[514,-25.362],[521,-25.6216],[526,-25.5591],[530,-25.2899],[533,-24.9128],[536,-24.3368],[538,-23.79],[539,-23.4481],[541,-22.6251],[542,-22.1364],[543,-21.5751],[544,-20.9496],[545,-20.2732],[546,-19.5248],[547,-18.701],[548,-17.8116],[549,-16.8591],[550,-15.8397],[551,-14.7584],[552,-13.6099],[553,-12.3962],[554,-11.1281],[555,-9.8026],[556,-8.423],[557,-6.9932],[558,-5.5258],[559,-4.0178],[560,-2.4813],[561,-0.9137],[562,0.6664],[563,2.2535],[564,3.8241],[565,5.3725],[566,6.8851],[567,8.3566],[568,9.7886],[569,11.1829],[570,12.5292],[571,13.8209],[572,15.0509],[573,16.2042],[574,17.2771],[575,18.2663],[576,19.1638],[577,19.9709],[578,20.7218],[579,21.4308],[581,22.7198],[583,23.8517],[585,24.8327],[587,25.6781],[589,26.3979],[591,26.9835],[595,27.8324],[599,28.4154],[606,28.9507],[615,29.2355],[617,29.205],[620,28.9499],[622,28.6243],[625,27.8909],[634,24.8804],[676,-6.5201],[677,-7.2762],[680,-8.1297],[685,-9.4459],[689,-10.4187],[699,-12.7937],[701,-13.2144],[703,-13.5654],[705,-13.8244],[707,-13.9856],[722,-14.3568],[730,-14.3445],[753,-13.5949],[755,-13.5869],[757,-13.6305],[759,-13.7262],[770,29.0942],[776,29.2346],[779,29.0604],[781,28.8058],[783,28.4073],[786,27.5953],[801,22.6063],[803,22.0176],[837,-6.5201],[887,24.5473],[894,22.3046],[895,22.0176],[929,-6.5201]]
       }
       hand_r_ik_ctrl.Rotation.Z {
        type=MovieSceneScriptingFloatChannel
        default=0
        thinned_from=751
        keys [[-2,142.0342],[27,146.78],[87,148.1946],[107,155.9058],[212,157.512],[217,156.0869],[239,140.0409],[247,136.6133],[296,134.648],[309,136.3923],[324,142.3644],[353,146.7916],[412,148.1946],[432,155.9058],[537,157.512],[542,156.0869],[564,140.0409],[572,136.6133],[621,134.648],[676,142.0342],[705,146.78],[759,147.6083],[770,134.1311],[803,137.8525],[837,142.0342],[887,136.5408],[895,137.8525],[929,142.0342]]
       }
       hand_r_ik_ctrl.Scale.X {type=MovieSceneScriptingFloatChannel, default=1, const [1,751]}
       hand_r_ik_ctrl.Scale.Y {type=MovieSceneScriptingFloatChannel, default=1, const [1,751]}
       hand_r_ik_ctrl.Scale.Z {type=MovieSceneScriptingFloatChannel, default=1, const [1,751]}
       Softness_2 {type=MovieSceneScriptingFloatChannel, default=0, const [0,751]}
       hand_r_ikorient_ctrl.Rotation.Roll {type=MovieSceneScriptingFloatChannel, default=0, const [0,445]}
       hand_r_ikorient_ctrl.Rotation.Pitch {type=MovieSceneScriptingFloatChannel, default=0, const [0,445]}
       hand_r_ikorient_ctrl.Rotation.Yaw {type=MovieSceneScriptingFloatChannel, default=0, const [0,445]}
       arm_r_pv_ik_ctrl.Location.X {
        type=MovieSceneScriptingFloatChannel
        default=0
        thinned_from=751
        keys [[-2,-4.962],[-1,-5.4314],[4,-6.3082],[14,-7.9457],[18,-8.6089],[20,-8.9054],[22,-9.1546],[24,-9.3257],[26,-9.4267],[27,-9.4534],[28,-9.4609],[38,-9.2657],[43,-9.2797],[47,-9.3698],[50,-9.4743],[56,-9.7822],[78,-11.1789],[82,-11.5006],[87,-11.811],[88,-11.9129],[89,-12.0662],[90,-12.2583],[91,-12.5007],[92,-12.7681],[94,-13.4051],[101,-15.889],[102,-16.2158],[103,-16.4993],[104,-16.7409],[107,-17.2982],[109,-17.5369],[110,-17.6285],[113,-17.7307],[116,-17.6688],[125,-17.1955],[133,-16.5433],[138,-16.2854],[152,-16.1112],[168,-16.4287],[171,-16.5822],[174,-16.8611],[176,-17.1556],[182,-18.2557],[184,-18.5204],[193,-19.3219],[199,-19.6569],[203,-19.7403],[206,-19.6939],[209,-19.5346],[211,-19.3588],[213,-19.0835],[214,-18.8643],[215,-18.5847],[216,-18.2573],[217,-17.8573],[218,-17.3654],[219,-16.792],[220,-16.1656],[221,-15.4695],[222,-14.7021],[223,-13.8794],[224,-12.9901],[225,-12.0468],[226,-11.0525],[227,-10.0175],[228,-8.9534],[232,-4.5809],[233,-3.4966],[234,-2.4238],[235,-1.3632],[236,-0.3236],[237,0.6877],[238,1.6553],[239,2.5858],[240,3.4581],[241,4.2792],[242,5.0352],[243,5.7274],[244,6.3504],[245,6.9086],[246,7.3948],[247,7.8178],[248,8.1783],[249,8.4844],[250,8.7256],[251,8.9191],[252,9.0652],[254,9.2788],[256,9.4083],[257,9.4506],[260,9.48],[270,9.3019],[286,8.9815],[291,8.8243],[293,8.7145],[295,8.5706],[297,8.3978],[299,8.1864],[302,7.8176],[307,7.131],[309,6.8401],[323,-4.962],[324,-5.4314],[329,-6.3082],[339,-7.9457],[343,-8.6089],[345,-8.9054],[347,-9.1546],[349,-9.3257],[351,-9.4267],[352,-9.4534],[353,-9.4609],[363,-9.2657],[368,-9.2797],[372,-9.3698],[375,-9.4743],[381,-9.7822],[403,-11.1789],[407,-11.5006],[412,-11.811],[413,-11.9129],[414,-12.0662],[415,-12.2583],[416,-12.5007],[417,-12.7681],[419,-13.4051],[426,-15.889],[427,-16.2158],[428,-16.4993],[429,-16.7409],[432,-17.2982],[434,-17.5369],[435,-17.6285],[438,-17.7307],[441,-17.6688],[450,-17.1955],[458,-16.5433],[463,-16.2854],[477,-16.1112],[493,-16.4287],[496,-16.5822],[499,-16.8611],[501,-17.1556],[507,-18.2557],[509,-18.5204],[518,-19.3219],[524,-19.6569],[528,-19.7403],[531,-19.6939],[534,-19.5346],[536,-19.3588],[538,-19.0835],[539,-18.8643],[540,-18.5847],[541,-18.2573],[542,-17.8573],[543,-17.3654],[544,-16.792],[545,-16.1656],[546,-15.4695],[547,-14.7021],[548,-13.8794],[549,-12.9901],[550,-12.0468],[551,-11.0525],[552,-10.0175],[553,-8.9534],[557,-4.5809],[558,-3.4966],[559,-2.4238],[560,-1.3632],[561,-0.3236],[562,0.6877],[563,1.6553],[564,2.5858],[565,3.4581],[566,4.2792],[567,5.0352],[568,5.7274],[569,6.3504],[570,6.9086],[571,7.3948],[572,7.8178],[573,8.1783],[574,8.4844],[575,8.7256],[576,8.9191],[577,9.0652],[579,9.2788],[581,9.4083],[582,9.4506],[585,9.48],[595,9.3019],[611,8.9815],[616,8.8243],[618,8.7145],[620,8.5706],[622,8.3978],[624,8.1864],[627,7.8176],[632,7.131],[634,6.8401],[676,-4.962],[677,-5.4314],[682,-6.3082],[692,-7.9457],[696,-8.6089],[698,-8.9054],[700,-9.1546],[702,-9.3257],[704,-9.4267],[705,-9.4534],[706,-9.4609],[716,-9.2657],[721,-9.2797],[725,-9.3698],[728,-9.4743],[734,-9.7822],[755,-11.1079],[759,-11.422],[770,8.9973],[775,8.8623],[776,8.8243],[778,8.7145],[780,8.5706],[782,8.3978],[784,8.1864],[787,7.8176],[792,7.131],[803,5.5371],[837,-4.962],[887,6.6973],[895,5.5371],[929,-4.962]]
       }
       arm_r_pv_ik_ctrl.Location.Y {
        type=MovieSceneScriptingFloatChannel
        default=0
        thinned_from=751
        keys [[-2,59.3404],[28,62.6224],[88,64.0574],[103,66.372],[133,67.0919],[170,64.5129],[214,66.2658],[220,65.829],[224,64.9917],[227,63.8888],[230,62.3105],[233,60.2655],[236,57.7058],[246,47.25],[248,45.3413],[250,43.6461],[252,42.2273],[255,40.4329],[258,38.9331],[262,37.3812],[265,36.4947],[270,35.5016],[276,34.8038],[290,34.3608],[293,34.4797],[296,34.8246],[299,35.4524],[309,38.5887],[323,59.3404],[353,62.6224],[413,64.0574],[428,66.372],[458,67.0919],[495,64.5129],[539,66.2658],[545,65.829],[549,64.9917],[552,63.8888],[555,62.3105],[558,60.2655],[561,57.7058],[571,47.25],[573,45.3413],[575,43.6461],[577,42.2273],[580,40.4329],[583,38.9331],[587,37.3812],[590,36.4947],[595,35.5016],[601,34.8038],[615,34.3608],[618,34.4797],[621,34.8246],[624,35.4524],[634,38.5887],[676,59.3404],[706,62.6224],[759,63.8654],[770,34.4433],[776,34.3731],[780,34.6786],[783,35.2149],[789,36.934],[803,41.3571],[837,59.3404],[887,38.9094],[895,41.3571],[929,59.3404]]
       }
       arm_r_pv_ik_ctrl.Location.Z {
        type=MovieSceneScriptingFloatChannel
        default=0
        thinned_from=751
        keys [[-2,47.6037],[-1,48.0711],[2,48.1924],[5,47.9368],[10,46.8758],[18,44.7311],[25,43.5356],[29,43.254],[35,43.5774],[55,46.0163],[64,46.3382],[69,46.0807],[82,43.8904],[94,40.9894],[98,40.3279],[103,39.9916],[106,40.1068],[110,40.6607],[114,41.548],[117,42.4851],[128,46.9376],[132,48.2035],[135,48.8059],[138,49.0339],[141,48.9356],[172,44.5224],[200,43.3947],[223,43.5768],[231,43.094],[248,40.6017],[253,40.3066],[273,41.3613],[286,42.0401],[294,41.8339],[308,40.2497],[309,40.0889],[323,47.6037],[324,48.0711],[327,48.1924],[330,47.9368],[335,46.8758],[343,44.7311],[350,43.5356],[354,43.254],[360,43.5774],[380,46.0163],[389,46.3382],[394,46.0807],[407,43.8904],[419,40.9894],[423,40.3279],[428,39.9916],[431,40.1068],[435,40.6607],[439,41.548],[442,42.4851],[453,46.9376],[457,48.2035],[460,48.8059],[463,49.0339],[466,48.9356],[497,44.5224],[525,43.3947],[548,43.5768],[556,43.094],[573,40.6017],[578,40.3066],[598,41.3613],[611,42.0401],[619,41.8339],[633,40.2497],[634,40.0889],[676,47.6037],[677,48.0711],[680,48.1924],[683,47.9368],[688,46.8758],[696,44.7311],[703,43.5356],[707,43.254],[713,43.5774],[733,46.0163],[742,46.3382],[747,46.0807],[755,44.8466],[770,42.0271],[779,41.8339],[793,40.2497],[803,38.5091],[837,47.6037],[895,38.5091],[929,47.6037]]
       }
       arm_r_pv_ik_ctrl.Rotation.X {
        type=MovieSceneScriptingFloatChannel
        default=0
        thinned_from=751
        keys [[-2,-1.766],[-1,-1.8281],[2,-1.998],[6,-2.2338],[7,-2.2893],[8,-2.3502],[9,-2.4336],[10,-2.5298],[11,-2.6424],[12,-2.7649],[13,-2.8994],[14,-3.044],[16,-3.3012],[17,-3.4114],[18,-3.5027],[19,-3.5814],[20,-3.6448],[21,-3.6907],[22,-3.7161],[23,-3.7243],[24,-3.7173],[35,-3.5061],[39,-3.4076],[43,-3.3277],[46,-3.2958],[48,-3.2892],[50,-3.2948],[51,-3.3208],[53,-3.3548],[56,-3.4338],[58,-3.5022],[59,-3.5432],[60,-3.576],[68,-3.9109],[72,-4.0527],[77,-4.1673],[79,-4.1955],[81,-4.2112],[82,-4.2284],[85,-4.3121],[86,-4.3516],[88,-4.447],[91,-4.553],[93,-4.5969],[101,-4.6869],[103,-4.6924],[104,-4.6864],[106,-4.6495],[111,-4.4871],[113,-4.4382],[114,-4.4004],[115,-4.3446],[116,-4.2733],[118,-4.0935],[119,-3.9968],[121,-3.7817],[124,-3.4181],[126,-3.1513],[131,-2.4431],[132,-2.3086],[133,-2.1878],[134,-2.0832],[135,-1.9958],[136,-1.9319],[137,-1.89],[138,-1.8663],[139,-1.8561],[140,-1.8534],[141,-1.8591],[142,-1.8712],[144,-1.9203],[147,-2.0208],[148,-2.0611],[150,-2.1289],[153,-2.2119],[155,-2.2435],[157,-2.2564],[159,-2.2476],[160,-2.2369],[161,-2.2208],[164,-2.1383],[167,-2.0405],[170,-1.9251],[172,-1.8343],[174,-1.7329],[175,-1.6789],[176,-1.6306],[177,-1.5785],[179,-1.4849],[180,-1.4451],[183,-1.3126],[185,-1.2184],[186,-1.175],[187,-1.1339],[188,-1.0986],[189,-1.0655],[190,-1.0367],[191,-1.0128],[192,-0.9933],[193,-0.9803],[195,-0.9622],[196,-0.9566],[198,-0.9572],[200,-0.9672],[201,-0.9767],[202,-0.9883],[204,-1.0227],[206,-1.0717],[207,-1.1014],[208,-1.1361],[209,-1.1812],[210,-1.2289],[211,-1.2846],[212,-1.3463],[213,-1.415],[214,-1.4733],[215,-1.5133],[216,-1.5373],[217,-1.539],[218,-1.5169],[219,-1.4704],[220,-1.3997],[221,-1.3241],[222,-1.2392],[223,-1.1459],[226,-0.8398],[227,-0.7482],[228,-0.6703],[229,-0.6167],[230,-0.5932],[231,-0.6112],[232,-0.6665],[233,-0.7567],[234,-0.8602],[235,-0.9729],[236,-1.0895],[237,-1.2139],[239,-1.478],[240,-1.6176],[241,-1.7611],[242,-1.9097],[243,-2.0674],[244,-2.2337],[247,-2.7515],[248,-2.917],[250,-3.2209],[251,-3.3602],[252,-3.4867],[253,-3.599],[255,-3.775],[257,-3.9722],[258,-4.0839],[262,-4.5716],[264,-4.7688],[266,-4.9242],[269,-5.0892],[273,-5.2559],[278,-5.3963],[296,-6.1519],[303,-6.5151],[308,-6.721],[309,-6.7706],[323,-1.766],[324,-1.8281],[327,-1.998],[331,-2.2338],[332,-2.2893],[333,-2.3502],[334,-2.4336],[335,-2.5298],[336,-2.6424],[337,-2.7649],[338,-2.8994],[339,-3.044],[341,-3.3012],[342,-3.4114],[343,-3.5027],[344,-3.5814],[345,-3.6448],[346,-3.6907],[347,-3.7161],[348,-3.7243],[349,-3.7173],[360,-3.5061],[364,-3.4076],[368,-3.3277],[371,-3.2958],[373,-3.2892],[375,-3.2948],[376,-3.3208],[378,-3.3548],[381,-3.4338],[383,-3.5022],[384,-3.5432],[385,-3.576],[393,-3.9109],[397,-4.0527],[402,-4.1673],[404,-4.1955],[406,-4.2112],[407,-4.2284],[410,-4.3121],[411,-4.3516],[413,-4.447],[416,-4.553],[418,-4.5969],[426,-4.6869],[428,-4.6924],[429,-4.6864],[431,-4.6495],[436,-4.4871],[438,-4.4382],[439,-4.4004],[440,-4.3446],[441,-4.2733],[443,-4.0935],[444,-3.9968],[446,-3.7817],[449,-3.4181],[451,-3.1513],[456,-2.4431],[457,-2.3086],[458,-2.1878],[459,-2.0832],[460,-1.9958],[461,-1.9319],[462,-1.89],[463,-1.8663],[464,-1.8561],[465,-1.8534],[466,-1.8591],[467,-1.8712],[469,-1.9203],[472,-2.0208],[473,-2.0611],[475,-2.1289],[478,-2.2119],[480,-2.2435],[482,-2.2564],[484,-2.2476],[485,-2.2369],[486,-2.2208],[489,-2.1383],[492,-2.0405],[495,-1.9251],[497,-1.8343],[499,-1.7329],[500,-1.6789],[501,-1.6306],[502,-1.5785],[504,-1.4849],[505,-1.4451],[508,-1.3126],[510,-1.2184],[511,-1.175],[512,-1.1339],[513,-1.0986],[514,-1.0655],[515,-1.0367],[516,-1.0128],[517,-0.9933],[518,-0.9803],[520,-0.9622],[521,-0.9566],[523,-0.9572],[525,-0.9672],[526,-0.9767],[527,-0.9883],[529,-1.0227],[531,-1.0717],[532,-1.1014],[533,-1.1361],[534,-1.1812],[535,-1.2289],[536,-1.2846],[537,-1.3463],[538,-1.415],[539,-1.4733],[540,-1.5133],[541,-1.5373],[542,-1.539],[543,-1.5169],[544,-1.4704],[545,-1.3997],[546,-1.3241],[547,-1.2392],[548,-1.1459],[551,-0.8398],[552,-0.7482],[553,-0.6703],[554,-0.6167],[555,-0.5932],[556,-0.6112],[557,-0.6665],[558,-0.7567],[559,-0.8602],[560,-0.9729],[561,-1.0895],[562,-1.2139],[564,-1.478],[565,-1.6176],[566,-1.7611],[567,-1.9097],[568,-2.0674],[569,-2.2337],[572,-2.7515],[573,-2.917],[575,-3.2209],[576,-3.3602],[577,-3.4867],[578,-3.599],[580,-3.775],[582,-3.9722],[583,-4.0839],[587,-4.5716],[589,-4.7688],[591,-4.9242],[594,-5.0892],[598,-5.2559],[603,-5.3963],[621,-6.1519],[628,-6.5151],[633,-6.721],[634,-6.7706],[676,-1.766],[677,-1.8281],[680,-1.998],[684,-2.2338],[685,-2.2893],[686,-2.3502],[687,-2.4336],[688,-2.5298],[689,-2.6424],[690,-2.7649],[691,-2.8994],[692,-3.044],[694,-3.3012],[695,-3.4114],[696,-3.5027],[697,-3.5814],[698,-3.6448],[699,-3.6907],[700,-3.7161],[701,-3.7243],[702,-3.7173],[713,-3.5061],[717,-3.4076],[721,-3.3277],[724,-3.2958],[726,-3.2892],[728,-3.2948],[729,-3.3208],[731,-3.3548],[734,-3.4338],[736,-3.5022],[737,-3.5432],[738,-3.576],[746,-3.9109],[750,-4.0527],[755,-4.1673],[757,-4.1955],[759,-4.2112],[770,-5.6776],[782,-6.2016],[788,-6.5151],[793,-6.721],[797,-6.9427],[799,-7.087],[803,-7.4306],[837,-1.766],[887,-6.823],[890,-7.0116],[895,-7.4306],[929,-1.766]]
       }
       arm_r_pv_ik_ctrl.Rotation.Y {
        type=MovieSceneScriptingFloatChannel
        default=0
        thinned_from=751
        keys [[-2,28.65],[-1,29.1585],[3,29.879],[11,30.8317],[22,32.796],[31,33.6887],[49,34.1731],[83,32.4271],[85,32.2059],[88,31.6179],[92,30.4895],[97,28.7215],[105,25.5222],[107,24.8211],[110,23.9606],[113,23.3272],[115,23.0476],[117,22.9047],[119,22.9081],[121,23.0594],[124,23.4826],[129,24.3787],[132,24.7769],[136,25.0678],[143,25.0921],[171,23.9717],[181,23.5107],[203,23.1943],[208,22.8941],[212,22.4247],[214,22.0852],[217,21.4204],[219,20.8693],[220,20.5545],[222,19.8032],[227,17.7216],[229,16.97],[230,16.6379],[231,16.3439],[233,15.8427],[235,15.4165],[237,15.0493],[238,14.9023],[239,14.8],[240,14.7435],[241,14.7391],[248,15.1531],[252,15.5511],[256,16.0734],[263,17.2588],[267,17.7933],[280,19.256],[287,19.7385],[293,19.8669],[309,19.715],[323,28.65],[324,29.1585],[328,29.879],[336,30.8317],[347,32.796],[356,33.6887],[374,34.1731],[408,32.4271],[410,32.2059],[413,31.6179],[417,30.4895],[422,28.7215],[430,25.5222],[432,24.8211],[435,23.9606],[438,23.3272],[440,23.0476],[442,22.9047],[444,22.9081],[446,23.0594],[449,23.4826],[454,24.3787],[457,24.7769],[461,25.0678],[468,25.0921],[496,23.9717],[506,23.5107],[528,23.1943],[533,22.8941],[537,22.4247],[539,22.0852],[542,21.4204],[544,20.8693],[545,20.5545],[547,19.8032],[552,17.7216],[554,16.97],[555,16.6379],[556,16.3439],[558,15.8427],[560,15.4165],[562,15.0493],[563,14.9023],[564,14.8],[565,14.7435],[566,14.7391],[573,15.1531],[577,15.5511],[581,16.0734],[588,17.2588],[592,17.7933],[605,19.256],[612,19.7385],[618,19.8669],[634,19.715],[676,28.65],[677,29.1585],[681,29.879],[689,30.8317],[700,32.796],[709,33.6887],[727,34.1731],[757,32.6218],[759,32.5549],[770,19.6329],[776,19.8509],[797,19.7363],[803,19.8899],[837,28.65],[887,19.7158],[895,19.8899],[929,28.65]]
       }
       arm_r_pv_ik_ctrl.Rotation.Z {
        type=MovieSceneScriptingFloatChannel
        default=0
        thinned_from=751
        keys [[-2,102.9282],[-1,103.571],[26,109.1287],[85,109.9197],[88,110.5449],[91,111.9734],[104,121.6548],[110,124.4804],[174,129.3506],[198,132.7654],[211,131.4472],[216,129.5753],[220,126.75],[224,122.7352],[228,117.4821],[233,109.3603],[243,91.0485],[246,86.0842],[248,83.1443],[250,80.5637],[253,77.4097],[257,73.9682],[261,71.154],[265,69.0169],[272,66.7326],[282,65.0146],[291,64.3759],[295,64.6674],[299,65.584],[309,69.4583],[323,102.9282],[324,103.571],[351,109.1287],[410,109.9197],[413,110.5449],[416,111.9734],[429,121.6548],[435,124.4804],[499,129.3506],[523,132.7654],[536,131.4472],[541,129.5753],[545,126.75],[549,122.7352],[553,117.4821],[558,109.3603],[568,91.0485],[571,86.0842],[573,83.1443],[575,80.5637],[578,77.4097],[582,73.9682],[586,71.154],[590,69.0169],[597,66.7326],[607,65.0146],[616,64.3759],[620,64.6674],[624,65.584],[634,69.4583],[676,102.9282],[677,103.571],[704,109.1287],[757,109.4374],[759,109.6119],[770,64.7378],[776,64.3759],[780,64.6674],[784,65.584],[803,72.6682],[837,102.9282],[887,69.8467],[895,72.6682],[929,102.9282]]
       }
       arm_r_pv_ik_ctrl.Scale.X {type=MovieSceneScriptingFloatChannel, default=1, const [1,751]}
       arm_r_pv_ik_ctrl.Scale.Y {type=MovieSceneScriptingFloatChannel, default=1, const [1,751]}
       arm_r_pv_ik_ctrl.Scale.Z {type=MovieSceneScriptingFloatChannel, default=1, const [1,751]}
       body_ctrl.Location.X {
        type=MovieSceneScriptingFloatChannel
        default=0
        thinned_from=600
        keys [[0,6.6514],[6,6.6042],[9,6.5357],[12,6.5267],[17,6.5972],[21,6.6723],[24,6.6723],[37,6.5163],[41,6.5214],[46,6.6053],[50,6.7296],[62,7.265],[72,7.7251],[76,7.8656],[80,7.9641],[84,7.9995],[85,7.9776],[86,7.9291],[87,7.8437],[88,7.7312],[89,7.5982],[91,7.2668],[93,6.8715],[94,6.6577],[95,6.4297],[97,5.9427],[102,4.63],[104,4.1288],[105,3.8966],[106,3.6756],[107,3.4766],[108,3.2903],[109,3.1117],[110,2.9515],[111,2.7989],[112,2.6571],[113,2.5349],[114,2.4225],[115,2.3217],[116,2.2297],[117,2.1515],[118,2.0841],[119,2.0238],[120,1.9813],[122,1.9093],[123,1.8821],[126,1.8312],[127,1.8227],[131,1.7597],[132,1.7402],[134,1.6905],[135,1.6592],[137,1.5846],[138,1.5442],[139,1.5002],[142,1.3576],[147,1.0918],[150,0.9436],[152,0.856],[155,0.737],[156,0.7007],[157,0.6681],[158,0.6337],[159,0.6039],[160,0.5766],[161,0.5482],[162,0.5232],[164,0.4789],[166,0.4429],[167,0.4271],[168,0.4139],[170,0.3941],[171,0.3911],[173,0.3883],[174,0.3921],[175,0.4002],[177,0.4194],[178,0.4324],[179,0.4464],[180,0.4656],[181,0.4802],[183,0.513],[184,0.5311],[185,0.5448],[186,0.564],[187,0.5869],[188,0.6085],[189,0.6254],[190,0.6491],[191,0.6672],[192,0.6909],[193,0.7174],[194,0.7371],[197,0.8208],[198,0.8501],[200,0.9028],[202,0.9656],[203,1.0029],[204,1.037],[205,1.0681],[206,1.1077],[207,1.1422],[209,1.218],[212,1.3377],[214,1.4218],[215,1.47],[218,1.6275],[220,1.7235],[221,1.758],[222,1.7841],[223,1.8034],[224,1.8098],[225,1.8096],[226,1.8027],[228,1.7728],[231,1.7066],[232,1.6805],[233,1.6499],[235,1.5803],[238,1.4674],[239,1.4318],[240,1.4031],[241,1.3779],[242,1.3616],[243,1.3529],[244,1.352],[245,1.3557],[246,1.3729],[247,1.3976],[248,1.4351],[249,1.4839],[250,1.5493],[251,1.6259],[252,1.7132],[254,1.9153],[256,2.1453],[257,2.2639],[259,2.5135],[262,2.8702],[264,3.0797],[266,3.2716],[269,3.5124],[271,3.669],[273,3.8046],[274,3.8782],[278,4.1311],[282,4.3339],[291,4.7159],[294,4.8641],[297,5.026],[300,5.2084],[316,5.8856],[349,6.6723],[362,6.5163],[366,6.5214],[371,6.6053],[375,6.7296],[387,7.265],[397,7.7251],[401,7.8656],[405,7.9641],[409,7.9995],[410,7.9776],[411,7.9291],[412,7.8437],[413,7.7312],[414,7.5982],[416,7.2668],[418,6.8715],[419,6.6577],[420,6.4297],[422,5.9427],[427,4.63],[429,4.1288],[430,3.8966],[431,3.6756],[432,3.4766],[433,3.2903],[434,3.1117],[435,2.9515],[436,2.7989],[437,2.6571],[438,2.5349],[439,2.4225],[440,2.3217],[441,2.2297],[442,2.1515],[443,2.0841],[444,2.0238],[445,1.9813],[447,1.9093],[448,1.8821],[451,1.8312],[452,1.8227],[456,1.7597],[457,1.7402],[459,1.6905],[460,1.6592],[462,1.5846],[463,1.5442],[464,1.5002],[467,1.3576],[472,1.0918],[475,0.9436],[477,0.856],[480,0.737],[481,0.7007],[482,0.6681],[483,0.6337],[484,0.6039],[485,0.5766],[486,0.5482],[487,0.5232],[489,0.4789],[491,0.4429],[492,0.4271],[493,0.4139],[495,0.3941],[496,0.3911],[498,0.3883],[499,0.3921],[500,0.4002],[502,0.4194],[503,0.4324],[504,0.4464],[505,0.4656],[506,0.4802],[508,0.513],[509,0.5311],[510,0.5448],[511,0.564],[512,0.5869],[513,0.6085],[514,0.6254],[515,0.6491],[516,0.6672],[517,0.6909],[518,0.7174],[519,0.7371],[522,0.8208],[523,0.8501],[525,0.9028],[527,0.9656],[528,1.0029],[529,1.037],[530,1.0681],[531,1.1077],[532,1.1422],[534,1.218],[537,1.3377],[539,1.4218],[540,1.47],[543,1.6275],[545,1.7235],[546,1.758],[547,1.7841],[548,1.8034],[549,1.8098],[550,1.8096],[551,1.8027],[553,1.7728],[556,1.7066],[557,1.6805],[558,1.6499],[560,1.5803],[563,1.4674],[564,1.4318],[565,1.4031],[566,1.3779],[567,1.3616],[568,1.3529],[569,1.352],[570,1.3557],[571,1.3729],[572,1.3976],[573,1.4351],[574,1.4839],[575,1.5493],[576,1.6259],[577,1.7132],[579,1.9153],[581,2.1453],[582,2.2639],[584,2.5135],[587,2.8702],[589,3.0797],[591,3.2716],[594,3.5124],[596,3.669],[598,3.8046],[599,3.8782],[603,4.1311],[607,4.3339],[616,4.7159],[619,4.8641],[622,5.026],[634,5.7637],[676,6.6514],[699,6.9835],[802,6.1746],[803,6.2242],[834,6.6453],[837,6.6514],[887,5.8181],[891,6.0351],[895,6.2242],[926,6.6453],[929,6.6514]]
       }
       body_ctrl.Location.Y {
        type=MovieSceneScriptingFloatChannel
        default=0
        thinned_from=731
        keys [[0,-0.8091],[6,-1.6755],[7,-1.7309],[8,-1.7816],[9,-1.8266],[10,-1.8673],[12,-1.9379],[15,-2.0204],[18,-2.0797],[21,-2.1231],[26,-2.1715],[31,-2.1927],[37,-2.1936],[39,-2.1814],[42,-2.1405],[43,-2.1237],[46,-2.0513],[52,-1.8692],[59,-1.6288],[63,-1.4939],[66,-1.3993],[67,-1.3657],[72,-1.2216],[74,-1.1705],[75,-1.1471],[76,-1.126],[78,-1.0938],[79,-1.0796],[80,-1.0753],[81,-1.0674],[83,-1.0668],[84,-1.0721],[85,-1.0913],[86,-1.1288],[87,-1.1832],[88,-1.2561],[89,-1.3444],[90,-1.4405],[91,-1.5534],[92,-1.6699],[93,-1.7937],[97,-2.3079],[98,-2.4325],[99,-2.5485],[100,-2.6589],[101,-2.7587],[102,-2.8446],[103,-2.9179],[104,-2.9772],[105,-3.0133],[106,-3.0345],[107,-3.0369],[109,-3.0045],[111,-2.9502],[114,-2.8343],[115,-2.7798],[116,-2.7169],[118,-2.6014],[121,-2.4227],[123,-2.3191],[124,-2.2796],[125,-2.2458],[127,-2.204],[128,-2.1937],[129,-2.1878],[130,-2.1915],[131,-2.2046],[132,-2.2225],[133,-2.2484],[134,-2.284],[135,-2.3251],[136,-2.3767],[138,-2.5023],[140,-2.6495],[142,-2.817],[145,-3.0952],[146,-3.1834],[147,-3.2814],[149,-3.4597],[150,-3.5541],[153,-3.8082],[157,-4.1085],[161,-4.3605],[164,-4.5132],[167,-4.6328],[169,-4.6975],[173,-4.7823],[181,-4.8563],[188,-4.8426],[192,-4.8135],[194,-4.7913],[195,-4.7706],[196,-4.7622],[200,-4.6795],[203,-4.5977],[205,-4.5314],[208,-4.4073],[212,-4.2059],[215,-4.0171],[217,-3.8742],[218,-3.7868],[219,-3.6896],[220,-3.5844],[222,-3.3542],[223,-3.2318],[226,-2.8367],[229,-2.4597],[232,-2.1052],[233,-1.9936],[235,-1.7852],[237,-1.6051],[238,-1.5243],[239,-1.4494],[240,-1.3874],[241,-1.3348],[242,-1.294],[243,-1.2628],[244,-1.2272],[245,-1.1995],[246,-1.1854],[247,-1.1551],[250,-1.0962],[251,-1.0798],[252,-1.0593],[254,-1.0102],[255,-0.9795],[258,-0.907],[259,-0.8813],[260,-0.8583],[261,-0.8322],[262,-0.8201],[263,-0.7984],[265,-0.764],[266,-0.749],[267,-0.736],[268,-0.7146],[269,-0.7024],[271,-0.6666],[272,-0.6411],[273,-0.6175],[274,-0.5884],[275,-0.5653],[277,-0.5129],[278,-0.4913],[279,-0.4666],[280,-0.4463],[282,-0.3998],[283,-0.3809],[284,-0.358],[285,-0.3395],[286,-0.3173],[288,-0.2806],[289,-0.2657],[290,-0.2498],[291,-0.2347],[292,-0.2179],[293,-0.2031],[296,-0.1661],[297,-0.1543],[298,-0.1442],[299,-0.1338],[300,-0.1211],[301,-0.1142],[303,-0.0901],[304,-0.079],[305,-0.069],[306,-0.0566],[307,-0.046],[308,-0.0331],[309,-0.0219],[331,-0.728],[332,-0.7537],[333,-0.7779],[335,-0.8211],[337,-0.8594],[340,-0.9093],[343,-0.9517],[350,-1.0329],[358,-1.105],[362,-1.135],[364,-1.1464],[368,-1.1585],[372,-1.1566],[394,-1.0935],[400,-1.088],[403,-1.0938],[404,-1.0796],[405,-1.0753],[406,-1.0674],[408,-1.0668],[409,-1.0721],[410,-1.0913],[411,-1.1288],[412,-1.1832],[413,-1.2561],[414,-1.3444],[415,-1.4405],[416,-1.5534],[417,-1.6699],[418,-1.7937],[422,-2.3079],[423,-2.4325],[424,-2.5485],[425,-2.6589],[426,-2.7587],[427,-2.8446],[428,-2.9179],[429,-2.9772],[430,-3.0133],[431,-3.0345],[432,-3.0369],[434,-3.0045],[436,-2.9502],[439,-2.8343],[440,-2.7798],[441,-2.7169],[443,-2.6014],[446,-2.4227],[448,-2.3191],[449,-2.2796],[450,-2.2458],[452,-2.204],[453,-2.1937],[454,-2.1878],[455,-2.1915],[456,-2.2046],[457,-2.2225],[458,-2.2484],[459,-2.284],[460,-2.3251],[461,-2.3767],[463,-2.5023],[465,-2.6495],[467,-2.817],[470,-3.0952],[471,-3.1834],[472,-3.2814],[474,-3.4597],[475,-3.5541],[478,-3.8082],[482,-4.1085],[486,-4.3605],[489,-4.5132],[492,-4.6328],[494,-4.6975],[498,-4.7823],[506,-4.8563],[513,-4.8426],[517,-4.8135],[519,-4.7913],[520,-4.7706],[521,-4.7622],[525,-4.6795],[528,-4.5977],[530,-4.5314],[533,-4.4073],[537,-4.2059],[540,-4.0171],[542,-3.8742],[543,-3.7868],[544,-3.6896],[545,-3.5844],[547,-3.3542],[548,-3.2318],[551,-2.8367],[554,-2.4597],[557,-2.1052],[558,-1.9936],[560,-1.7852],[562,-1.6051],[563,-1.5243],[564,-1.4494],[565,-1.3874],[566,-1.3348],[567,-1.294],[568,-1.2628],[569,-1.2272],[570,-1.1995],[571,-1.1854],[572,-1.1551],[575,-1.0962],[576,-1.0798],[577,-1.0593],[579,-1.0102],[580,-0.9795],[583,-0.907],[584,-0.8813],[585,-0.8583],[586,-0.8322],[587,-0.8201],[588,-0.7984],[590,-0.764],[591,-0.749],[592,-0.736],[593,-0.7146],[594,-0.7024],[596,-0.6666],[597,-0.6411],[598,-0.6175],[599,-0.5884],[600,-0.5653],[602,-0.5129],[603,-0.4913],[604,-0.4666],[605,-0.4463],[607,-0.3998],[608,-0.3809],[609,-0.358],[610,-0.3395],[611,-0.3173],[613,-0.2806],[614,-0.2657],[615,-0.2498],[616,-0.2347],[617,-0.2179],[618,-0.2031],[621,-0.1661],[622,-0.1543],[623,-0.1442],[624,-0.1338],[625,-0.1211],[626,-0.1142],[628,-0.0901],[629,-0.079],[630,-0.069],[631,-0.0566],[632,-0.046],[633,-0.0331],[634,-0.0219],[676,-0.8091],[684,-1.6755],[685,-1.7309],[686,-1.7816],[687,-1.8266],[688,-1.8673],[690,-1.9379],[693,-2.0204],[696,-2.0797],[699,-2.1231],[704,-2.1715],[709,-2.1927],[715,-2.1936],[717,-2.1814],[720,-2.1405],[721,-2.1237],[724,-2.0513],[730,-1.8692],[737,-1.6288],[741,-1.4939],[744,-1.3993],[745,-1.3657],[750,-1.2216],[752,-1.1705],[753,-1.1471],[754,-1.126],[755,-1.109],[757,-1.0796],[759,-1.0674],[770,-0.3395],[771,-0.3173],[773,-0.2806],[774,-0.2657],[775,-0.2498],[776,-0.2347],[777,-0.2179],[778,-0.2031],[781,-0.1661],[782,-0.1543],[783,-0.1442],[784,-0.1338],[785,-0.1211],[786,-0.1142],[788,-0.0901],[789,-0.079],[790,-0.069],[791,-0.0566],[792,-0.046],[793,-0.0331],[794,-0.0219],[795,-0.0082],[796,0.0042],[797,0.017],[798,0.0319],[799,0.0438],[800,0.0584],[801,0.0734],[802,0.0849],[803,0.1024],[834,-0.7891],[837,-0.8091],[887,-0.0082],[888,0.0042],[889,0.017],[890,0.0319],[891,0.0438],[892,0.0584],[893,0.0734],[894,0.0849],[895,0.1024],[926,-0.7891],[929,-0.8091]]
       }
       body_ctrl.Location.Z {
        type=MovieSceneScriptingFloatChannel
        default=0
        keys [[0,38.2052],[52,4.9853],[127,50.0283],[203,11.709],[235,24.8361],[280,59.0968],[294,62.0499],[328,40.0373],[377,4.9853],[452,50.0283],[528,11.709],[560,24.8361],[605,59.0968],[619,62.0499],[668,36.9921],[709,15.3868],[780,64.4567],[785,63.6707],[829,38.0435],[844,32.1376],[894,64.5375],[926,41.0773],[929,38.2052]]
       }
       body_ctrl.Rotation.X {
        type=MovieSceneScriptingFloatChannel
        default=0
        thinned_from=589
        keys [[0,-9.4086],[40,-9.7743],[49,-10.0266],[56,-10.3878],[71,-11.4142],[85,-12.0056],[88,-12.0621],[91,-12.0288],[94,-11.9161],[107,-10.9993],[112,-10.7636],[115,-10.6492],[118,-10.4365],[123,-10.0255],[126,-9.8411],[128,-9.7611],[130,-9.722],[132,-9.7331],[134,-9.7861],[136,-9.8963],[153,-11.3725],[159,-11.7507],[162,-11.8787],[165,-11.942],[168,-11.9281],[171,-11.8365],[176,-11.5457],[191,-10.3989],[204,-9.6403],[208,-9.3348],[211,-9.0413],[213,-8.8061],[215,-8.514],[216,-8.3401],[217,-8.1384],[218,-7.9081],[219,-7.653],[223,-6.552],[225,-6.0388],[226,-5.8034],[227,-5.586],[228,-5.3918],[229,-5.2136],[230,-5.0605],[231,-4.9294],[232,-4.8255],[233,-4.7457],[234,-4.6925],[235,-4.6621],[236,-4.6585],[237,-4.6706],[238,-4.7117],[239,-4.7849],[240,-4.8932],[241,-5.0406],[242,-5.2248],[246,-6.0447],[248,-6.4289],[250,-6.7816],[252,-7.0867],[253,-7.219],[255,-7.4418],[257,-7.6228],[259,-7.7679],[261,-7.8678],[262,-7.8974],[264,-7.9121],[266,-7.8881],[268,-7.818],[270,-7.7093],[272,-7.5624],[274,-7.3692],[288,-5.9002],[291,-5.6353],[294,-5.4013],[297,-5.204],[300,-5.0449],[303,-4.9224],[305,-4.8613],[308,-4.8064],[309,-4.799],[330,-6.6563],[365,-9.7743],[374,-10.0266],[381,-10.3878],[396,-11.4142],[410,-12.0056],[413,-12.0621],[416,-12.0288],[419,-11.9161],[432,-10.9993],[437,-10.7636],[440,-10.6492],[443,-10.4365],[448,-10.0255],[451,-9.8411],[453,-9.7611],[455,-9.722],[457,-9.7331],[459,-9.7861],[461,-9.8963],[478,-11.3725],[484,-11.7507],[487,-11.8787],[490,-11.942],[493,-11.9281],[496,-11.8365],[501,-11.5457],[516,-10.3989],[529,-9.6403],[533,-9.3348],[536,-9.0413],[538,-8.8061],[540,-8.514],[541,-8.3401],[542,-8.1384],[543,-7.9081],[544,-7.653],[548,-6.552],[550,-6.0388],[551,-5.8034],[552,-5.586],[553,-5.3918],[554,-5.2136],[555,-5.0605],[556,-4.9294],[557,-4.8255],[558,-4.7457],[559,-4.6925],[560,-4.6621],[561,-4.6585],[562,-4.6706],[563,-4.7117],[564,-4.7849],[565,-4.8932],[566,-5.0406],[567,-5.2248],[571,-6.0447],[573,-6.4289],[575,-6.7816],[577,-7.0867],[578,-7.219],[580,-7.4418],[582,-7.6228],[584,-7.7679],[586,-7.8678],[587,-7.8974],[589,-7.9121],[591,-7.8881],[593,-7.818],[595,-7.7093],[597,-7.5624],[599,-7.3692],[613,-5.9002],[616,-5.6353],[619,-5.4013],[622,-5.204],[625,-5.0449],[628,-4.9224],[630,-4.8613],[633,-4.8064],[634,-4.799],[676,-3.3415],[718,-2.8608],[724,-2.7807],[726,-2.7636],[727,-2.7595],[729,-2.7615],[732,-2.7862],[736,-2.8552],[744,-3.048],[747,-3.1044],[749,-3.1292],[752,-3.1369],[755,-3.1148],[759,-3.0663],[834,-3.3603],[837,-3.3415],[887,-8.5308],[897,-9.1467],[926,-9.4275],[929,-9.4086]]
       }
       body_ctrl.Rotation.Y {
        type=MovieSceneScriptingFloatChannel
        default=0
        thinned_from=511
        keys [[0,0.021],[15,-2.4082],[18,-3.0815],[19,-3.2997],[20,-3.5065],[21,-3.698],[22,-3.868],[23,-4.0097],[24,-4.1169],[25,-4.2021],[26,-4.2769],[27,-4.3402],[28,-4.3921],[30,-4.4696],[33,-4.5377],[39,-4.6023],[47,-4.6671],[54,-4.7606],[70,-4.8893],[74,-4.947],[87,-4.9101],[93,-4.9455],[95,-4.936],[96,-4.9177],[97,-4.8873],[99,-4.7977],[101,-4.6727],[102,-4.6213],[103,-4.4237],[104,-4.2105],[106,-3.7545],[109,-3.0596],[110,-2.8407],[111,-2.628],[112,-2.4254],[113,-2.2346],[114,-2.0504],[115,-1.8826],[116,-1.7315],[117,-1.602],[118,-1.4946],[119,-1.4098],[120,-1.3514],[121,-1.3158],[122,-1.2974],[123,-1.2991],[124,-1.3178],[125,-1.3561],[126,-1.4082],[127,-1.4718],[128,-1.5453],[133,-1.9367],[137,-2.2294],[144,-2.7064],[162,-3.865],[175,-4.5787],[182,-5.0121],[190,-5.3921],[197,-5.6432],[203,-5.788],[208,-5.8445],[211,-5.8377],[218,-5.7221],[219,-5.6873],[220,-5.6344],[221,-5.5542],[222,-5.4512],[223,-5.3304],[227,-4.7698],[228,-4.6381],[229,-4.5201],[267,-1.5785],[295,-1.0653],[325,0.021],[340,-2.4082],[343,-3.0815],[344,-3.2997],[345,-3.5065],[346,-3.698],[347,-3.868],[348,-4.0097],[349,-4.1169],[350,-4.2021],[351,-4.2769],[352,-4.3402],[353,-4.3921],[355,-4.4696],[358,-4.5377],[364,-4.6023],[372,-4.6671],[379,-4.7606],[386,-4.8121],[410,-4.7346],[418,-4.7872],[420,-4.7813],[421,-4.7648],[422,-4.7362],[424,-4.6502],[426,-4.5288],[428,-4.4237],[429,-4.2105],[431,-3.7545],[434,-3.0596],[435,-2.8407],[436,-2.628],[437,-2.4254],[438,-2.2346],[439,-2.0504],[440,-1.8826],[441,-1.7315],[442,-1.602],[443,-1.4946],[444,-1.4098],[445,-1.3514],[446,-1.3158],[447,-1.2974],[448,-1.2991],[449,-1.3178],[450,-1.3561],[451,-1.4082],[452,-1.4718],[453,-1.5453],[458,-1.9367],[462,-2.2294],[469,-2.7064],[487,-3.865],[500,-4.5787],[507,-5.0121],[515,-5.3921],[522,-5.6432],[528,-5.788],[533,-5.8445],[536,-5.8377],[543,-5.7221],[544,-5.6873],[545,-5.6344],[546,-5.5542],[547,-5.4512],[548,-5.3304],[552,-4.7698],[553,-4.6381],[554,-4.5201],[592,-1.5785],[620,-1.0653],[676,-1.0272],[693,-3.4564],[696,-4.1296],[697,-4.3479],[698,-4.5546],[699,-4.7461],[700,-4.9162],[701,-5.0578],[702,-5.1651],[704,-5.3251],[705,-5.3884],[707,-5.4831],[709,-5.5467],[714,-5.6246],[725,-5.7153],[747,-5.9264],[798,-1.7339],[799,-1.7154],[800,-1.6895],[802,-1.6272],[803,-1.5916],[834,-1.0286],[837,-1.0272],[887,-1.6324],[889,-1.5298],[890,-1.4761],[891,-1.4193],[895,-1.1753],[926,0.0196],[929,0.021]]
       }
       body_ctrl.Rotation.Z {
        type=MovieSceneScriptingFloatChannel
        default=0
        thinned_from=517
        keys [[0,-17.3591],[52,-16.1779],[75,-16.38],[76,-16.4079],[88,-15.4996],[90,-14.9231],[91,-14.6878],[92,-14.5139],[93,-14.4165],[94,-14.4165],[95,-14.5319],[96,-14.7825],[97,-15.171],[98,-15.6802],[99,-16.2948],[100,-17.0018],[101,-17.7825],[102,-18.623],[107,-23.2028],[109,-24.9331],[110,-25.7274],[111,-26.4582],[113,-27.715],[115,-28.8542],[116,-29.3563],[117,-29.793],[118,-30.1497],[119,-30.4109],[120,-30.5589],[122,-30.6323],[125,-30.5172],[153,-27.9989],[174,-29.7718],[206,-32.9099],[216,-33.2473],[220,-33.092],[223,-33.0914],[226,-32.8348],[228,-32.4631],[230,-31.8602],[231,-31.4595],[232,-30.9838],[233,-30.4282],[234,-29.8002],[236,-28.734],[239,-26.9106],[246,-22.3095],[248,-21.0737],[250,-19.937],[252,-18.9289],[255,-17.5847],[257,-16.7836],[259,-16.07],[261,-15.4499],[263,-14.9297],[265,-14.521],[266,-14.3633],[271,-13.8552],[276,-13.4959],[282,-13.2456],[290,-13.1096],[293,-13.1379],[298,-13.355],[309,-14.0958],[325,-17.3591],[377,-16.1779],[400,-16.38],[401,-16.4079],[413,-15.4996],[415,-14.9231],[416,-14.6878],[417,-14.5139],[418,-14.4165],[419,-14.4165],[420,-14.5319],[421,-14.7825],[422,-15.171],[423,-15.6802],[424,-16.2948],[425,-17.0018],[426,-17.7825],[427,-18.623],[432,-23.2028],[434,-24.9331],[435,-25.7274],[436,-26.4582],[438,-27.715],[440,-28.8542],[441,-29.3563],[442,-29.793],[443,-30.1497],[444,-30.4109],[445,-30.5589],[447,-30.6323],[450,-30.5172],[478,-27.9989],[499,-29.7718],[531,-32.9099],[541,-33.2473],[545,-33.092],[548,-33.0914],[551,-32.8348],[553,-32.4631],[555,-31.8602],[556,-31.4595],[557,-30.9838],[558,-30.4282],[559,-29.8002],[561,-28.734],[564,-26.9106],[571,-22.3095],[573,-21.0737],[575,-19.937],[577,-18.9289],[580,-17.5847],[582,-16.7836],[584,-16.07],[586,-15.4499],[588,-14.9297],[590,-14.521],[591,-14.3633],[596,-13.8552],[601,-13.4959],[607,-13.2456],[615,-13.1096],[618,-13.1379],[623,-13.355],[634,-14.0958],[676,-17.1479],[731,-15.9607],[753,-16.1688],[754,-16.1967],[776,-15.8886],[780,-15.9839],[803,-17.3807],[834,-17.115],[837,-17.1479],[887,-14.1518],[895,-14.6004],[926,-17.3262],[929,-17.3591]]
       }
       body_ctrl.Scale.X {type=MovieSceneScriptingFloatChannel, default=1, const [1,732]}
       body_ctrl.Scale.Y {type=MovieSceneScriptingFloatChannel, default=1, const [1,732]}
       body_ctrl.Scale.Z {type=MovieSceneScriptingFloatChannel, default=1, const [1,732]}
       hips_ctrl.Location.X {
        type=MovieSceneScriptingFloatChannel
        default=0
        thinned_from=752
        keys [[-2,-0.1208],[-1,-0.128],[2,-0.1334],[4,-0.1365],[5,-0.1394],[6,-0.1408],[7,-0.1436],[8,-0.1479],[9,-0.1535],[10,-0.1606],[11,-0.1692],[12,-0.1792],[13,-0.1906],[14,-0.2034],[15,-0.2212],[17,-0.2483],[18,-0.2683],[19,-0.2819],[21,-0.322],[26,-0.3897],[27,-0.3965],[28,-0.41],[30,-0.424],[31,-0.4265],[33,-0.4355],[34,-0.438],[35,-0.4425],[40,-0.4548],[41,-0.4552],[42,-0.4577],[43,-0.4581],[44,-0.4605],[49,-0.4614],[50,-0.4585],[51,-0.4587],[57,-0.4403],[58,-0.4342],[59,-0.4309],[60,-0.4247],[61,-0.4215],[62,-0.4153],[63,-0.4116],[75,-0.3525],[76,-0.3521],[78,-0.3427],[79,-0.3418],[80,-0.3418],[81,-0.3361],[85,-0.3344],[88,-0.3493],[92,-0.3926],[93,-0.3982],[94,-0.4091],[95,-0.415],[97,-0.4147],[99,-0.396],[100,-0.3822],[102,-0.3447],[104,-0.2978],[110,-0.1263],[111,-0.0997],[112,-0.067],[113,-0.0444],[114,-0.0229],[116,0.0223],[119,0.0564],[121,0.0588],[122,0.059],[123,0.0602],[125,0.0402],[126,0.0292],[127,0.0134],[128,0.0019],[129,-0.0099],[131,-0.0331],[132,-0.0445],[134,-0.056],[135,-0.0614],[136,-0.0614],[137,-0.0666],[138,-0.0664],[139,-0.0658],[140,-0.0598],[141,-0.0594],[142,-0.0588],[143,-0.0539],[145,-0.0458],[146,-0.0412],[149,-0.0291],[150,-0.0255],[152,-0.0174],[153,-0.0169],[156,-0.0046],[157,-0.0047],[158,-0.0006],[159,-0.0002],[160,0.0013],[161,0.0013],[162,0.0028],[168,0.0051],[169,0.0043],[170,0.0048],[171,0.0033],[172,0.0015],[173,-0.0008],[174,-0.0049],[175,-0.0105],[179,-0.034],[180,-0.0461],[184,-0.0711],[185,-0.0714],[186,-0.0781],[187,-0.0843],[188,-0.0909],[189,-0.0913],[190,-0.098],[191,-0.098],[192,-0.1024],[193,-0.1058],[194,-0.1068],[195,-0.1068],[196,-0.1112],[197,-0.1113],[198,-0.1123],[199,-0.1123],[200,-0.11],[201,-0.1099],[202,-0.1076],[203,-0.1076],[204,-0.1042],[205,-0.1019],[206,-0.0985],[207,-0.0913],[208,-0.0879],[209,-0.0804],[210,-0.077],[211,-0.0695],[212,-0.0584],[213,-0.0507],[214,-0.0393],[216,-0.0235],[218,-0.0153],[219,-0.007],[220,-0.0024],[221,0.0098],[222,0.0225],[223,0.0345],[224,0.0512],[225,0.0682],[226,0.0849],[228,0.1177],[229,0.1335],[231,0.1635],[232,0.1781],[238,0.2614],[239,0.2773],[244,0.3423],[245,0.3583],[248,0.3984],[249,0.4089],[251,0.4247],[253,0.4345],[254,0.4368],[256,0.4351],[258,0.4314],[260,0.4259],[261,0.4241],[262,0.4203],[264,0.4165],[265,0.4155],[267,0.4153],[268,0.4171],[269,0.417],[270,0.4159],[271,0.4157],[273,0.4095],[274,0.4089],[280,0.3895],[281,0.3889],[286,0.3745],[287,0.3704],[291,0.3698],[292,0.3626],[294,0.3631],[295,0.3559],[297,0.3569],[298,0.3497],[299,0.3502],[302,0.3295],[307,0.3015],[308,0.29],[309,0.2843],[323,-0.1208],[324,-0.128],[327,-0.1334],[329,-0.1365],[330,-0.1394],[331,-0.1408],[332,-0.1436],[333,-0.1479],[334,-0.1535],[335,-0.1606],[336,-0.1692],[337,-0.1792],[338,-0.1906],[339,-0.2034],[340,-0.2212],[342,-0.2483],[343,-0.2683],[344,-0.2819],[346,-0.322],[351,-0.3897],[352,-0.3965],[353,-0.41],[355,-0.424],[356,-0.4265],[358,-0.4355],[359,-0.438],[360,-0.4425],[365,-0.4548],[366,-0.4552],[367,-0.4577],[368,-0.4581],[369,-0.4605],[374,-0.4614],[375,-0.4585],[376,-0.4587],[382,-0.4403],[383,-0.4342],[384,-0.4309],[385,-0.4247],[386,-0.4215],[387,-0.4153],[388,-0.4116],[400,-0.3525],[401,-0.3521],[403,-0.3427],[404,-0.3418],[405,-0.3418],[406,-0.3361],[410,-0.3344],[413,-0.3493],[417,-0.3926],[418,-0.3982],[419,-0.4091],[420,-0.415],[422,-0.4147],[424,-0.396],[425,-0.3822],[427,-0.3447],[429,-0.2978],[435,-0.1263],[436,-0.0997],[437,-0.067],[438,-0.0444],[439,-0.0229],[441,0.0223],[444,0.0564],[446,0.0588],[447,0.059],[448,0.0602],[450,0.0402],[451,0.0292],[452,0.0134],[453,0.0019],[454,-0.0099],[456,-0.0331],[457,-0.0445],[459,-0.056],[460,-0.0614],[461,-0.0614],[462,-0.0666],[463,-0.0664],[464,-0.0658],[465,-0.0598],[466,-0.0594],[467,-0.0588],[468,-0.0539],[470,-0.0458],[471,-0.0412],[474,-0.0291],[475,-0.0255],[477,-0.0174],[478,-0.0169],[481,-0.0046],[482,-0.0047],[483,-0.0006],[484,-0.0002],[485,0.0013],[486,0.0013],[487,0.0028],[493,0.0051],[494,0.0043],[495,0.0048],[496,0.0033],[497,0.0015],[498,-0.0008],[499,-0.0049],[500,-0.0105],[504,-0.034],[505,-0.0461],[509,-0.0711],[510,-0.0714],[511,-0.0781],[512,-0.0843],[513,-0.0909],[514,-0.0913],[515,-0.098],[516,-0.098],[517,-0.1024],[518,-0.1058],[519,-0.1068],[520,-0.1068],[521,-0.1112],[522,-0.1113],[523,-0.1123],[524,-0.1123],[525,-0.11],[526,-0.1099],[527,-0.1076],[528,-0.1076],[529,-0.1042],[530,-0.1019],[531,-0.0985],[532,-0.0913],[533,-0.0879],[534,-0.0804],[535,-0.077],[536,-0.0695],[537,-0.0584],[538,-0.0507],[539,-0.0393],[541,-0.0235],[543,-0.0153],[544,-0.007],[545,-0.0024],[546,0.0098],[547,0.0225],[548,0.0345],[549,0.0512],[550,0.0682],[551,0.0849],[553,0.1177],[554,0.1335],[556,0.1635],[557,0.1781],[563,0.2614],[564,0.2773],[569,0.3423],[570,0.3583],[573,0.3984],[574,0.4089],[576,0.4247],[578,0.4345],[579,0.4368],[581,0.4351],[583,0.4314],[585,0.4259],[586,0.4241],[587,0.4203],[589,0.4165],[590,0.4155],[592,0.4153],[593,0.4171],[594,0.417],[595,0.4159],[596,0.4157],[598,0.4095],[599,0.4089],[605,0.3895],[606,0.3889],[611,0.3745],[612,0.3704],[616,0.3698],[617,0.3626],[619,0.3631],[620,0.3559],[622,0.3569],[623,0.3497],[624,0.3502],[627,0.3295],[632,0.3015],[633,0.29],[634,0.2843],[676,-0.1208],[677,-0.128],[680,-0.1334],[682,-0.1365],[683,-0.1394],[684,-0.1408],[685,-0.1436],[686,-0.1479],[687,-0.1535],[688,-0.1606],[689,-0.1692],[690,-0.1792],[691,-0.1906],[692,-0.2034],[693,-0.2212],[695,-0.2483],[696,-0.2683],[697,-0.2819],[699,-0.322],[704,-0.3897],[705,-0.3965],[706,-0.41],[708,-0.424],[709,-0.4265],[711,-0.4355],[712,-0.438],[713,-0.4425],[718,-0.4548],[719,-0.4552],[720,-0.4577],[721,-0.4581],[722,-0.4605],[727,-0.4614],[728,-0.4585],[729,-0.4587],[735,-0.4403],[736,-0.4342],[737,-0.4309],[738,-0.4247],[739,-0.4215],[740,-0.4153],[741,-0.4116],[753,-0.3525],[754,-0.3521],[755,-0.3476],[759,-0.3361],[770,0.3777],[771,0.3745],[772,0.3704],[776,0.3698],[777,0.3626],[779,0.3631],[780,0.3559],[782,0.3569],[783,0.3497],[784,0.3502],[787,0.3295],[792,0.3015],[793,0.29],[798,0.2612],[799,0.2551],[800,0.2555],[801,0.2494],[802,0.2494],[803,0.2433],[837,-0.1208],[887,0.2785],[891,0.2551],[892,0.2555],[893,0.2494],[894,0.2494],[895,0.2433],[927,-0.1182],[929,-0.1208]]
       }
       hips_ctrl.Location.Y {
        type=MovieSceneScriptingFloatChannel
        default=0
        thinned_from=752
        keys [[-2,-0.9889],[-1,-0.977],[2,-0.9677],[10,-0.9578],[14,-0.9472],[28,-0.9251],[44,-0.8808],[45,-0.8808],[51,-0.8718],[55,-0.8712],[75,-0.9038],[76,-0.9064],[77,-0.9057],[79,-0.9102],[83,-0.915],[87,-0.9312],[90,-0.9386],[91,-0.9382],[92,-0.9398],[93,-0.9386],[94,-0.9401],[96,-0.939],[98,-0.9423],[101,-0.9525],[103,-0.9649],[106,-0.9927],[110,-1.0409],[112,-1.062],[113,-1.0746],[114,-1.0795],[116,-1.1048],[121,-1.139],[122,-1.139],[123,-1.1467],[125,-1.1494],[126,-1.1431],[127,-1.1449],[131,-1.1379],[134,-1.1366],[135,-1.1389],[136,-1.1389],[138,-1.1441],[142,-1.158],[143,-1.1588],[145,-1.1652],[146,-1.172],[149,-1.1815],[150,-1.181],[156,-1.2006],[157,-1.2006],[160,-1.2084],[161,-1.2084],[170,-1.2349],[171,-1.2411],[172,-1.2443],[173,-1.2446],[175,-1.251],[179,-1.2542],[185,-1.2422],[186,-1.2374],[187,-1.2354],[190,-1.2229],[191,-1.2229],[192,-1.2158],[193,-1.2162],[194,-1.2086],[195,-1.2086],[196,-1.2015],[197,-1.2015],[198,-1.1939],[199,-1.1939],[200,-1.1858],[201,-1.1858],[202,-1.1777],[204,-1.1773],[205,-1.1692],[212,-1.1594],[214,-1.1605],[218,-1.1721],[221,-1.1929],[226,-1.249],[228,-1.2644],[229,-1.2672],[231,-1.2607],[232,-1.2538],[235,-1.2183],[239,-1.1612],[241,-1.1294],[242,-1.1108],[246,-1.0467],[249,-1.0076],[253,-0.9701],[254,-0.9642],[255,-0.9624],[261,-0.9183],[262,-0.9035],[267,-0.8661],[268,-0.8658],[271,-0.8444],[274,-0.8318],[275,-0.8234],[281,-0.7986],[282,-0.7989],[284,-0.7908],[285,-0.7911],[286,-0.7871],[288,-0.7861],[289,-0.7821],[292,-0.7832],[293,-0.7871],[295,-0.7881],[297,-0.7961],[298,-0.7971],[300,-0.806],[301,-0.807],[304,-0.8228],[305,-0.8263],[306,-0.8323],[307,-0.8357],[308,-0.8426],[309,-0.846],[323,-0.9889],[324,-0.977],[327,-0.9677],[335,-0.9578],[339,-0.9472],[353,-0.9251],[369,-0.8808],[370,-0.8808],[376,-0.8718],[380,-0.8712],[400,-0.9038],[401,-0.9064],[402,-0.9057],[404,-0.9102],[408,-0.915],[412,-0.9312],[415,-0.9386],[416,-0.9382],[417,-0.9398],[418,-0.9386],[419,-0.9401],[421,-0.939],[423,-0.9423],[426,-0.9525],[428,-0.9649],[431,-0.9927],[435,-1.0409],[437,-1.062],[438,-1.0746],[439,-1.0795],[441,-1.1048],[446,-1.139],[447,-1.139],[448,-1.1467],[450,-1.1494],[451,-1.1431],[452,-1.1449],[456,-1.1379],[459,-1.1366],[460,-1.1389],[461,-1.1389],[463,-1.1441],[467,-1.158],[468,-1.1588],[470,-1.1652],[471,-1.172],[474,-1.1815],[475,-1.181],[481,-1.2006],[482,-1.2006],[485,-1.2084],[486,-1.2084],[495,-1.2349],[496,-1.2411],[497,-1.2443],[498,-1.2446],[500,-1.251],[504,-1.2542],[510,-1.2422],[511,-1.2374],[512,-1.2354],[515,-1.2229],[516,-1.2229],[517,-1.2158],[518,-1.2162],[519,-1.2086],[520,-1.2086],[521,-1.2015],[522,-1.2015],[523,-1.1939],[524,-1.1939],[525,-1.1858],[526,-1.1858],[527,-1.1777],[529,-1.1773],[530,-1.1692],[537,-1.1594],[539,-1.1605],[543,-1.1721],[546,-1.1929],[551,-1.249],[553,-1.2644],[554,-1.2672],[556,-1.2607],[557,-1.2538],[560,-1.2183],[564,-1.1612],[566,-1.1294],[567,-1.1108],[571,-1.0467],[574,-1.0076],[578,-0.9701],[579,-0.9642],[580,-0.9624],[586,-0.9183],[587,-0.9035],[592,-0.8661],[593,-0.8658],[596,-0.8444],[599,-0.8318],[600,-0.8234],[606,-0.7986],[607,-0.7989],[609,-0.7908],[610,-0.7911],[611,-0.7871],[613,-0.7861],[614,-0.7821],[617,-0.7832],[618,-0.7871],[620,-0.7881],[622,-0.7961],[623,-0.7971],[625,-0.806],[626,-0.807],[629,-0.8228],[630,-0.8263],[631,-0.8323],[632,-0.8357],[633,-0.8426],[676,-0.9889],[677,-0.977],[680,-0.9677],[688,-0.9578],[692,-0.9472],[706,-0.9251],[722,-0.8808],[723,-0.8808],[729,-0.8718],[733,-0.8712],[753,-0.9038],[754,-0.9064],[755,-0.9057],[757,-0.9102],[759,-0.9112],[770,-0.7911],[771,-0.7871],[773,-0.7861],[774,-0.7821],[777,-0.7832],[778,-0.7871],[780,-0.7881],[782,-0.7961],[783,-0.7971],[785,-0.806],[786,-0.807],[789,-0.8228],[790,-0.8263],[791,-0.8323],[792,-0.8357],[793,-0.8426],[798,-0.8598],[799,-0.8607],[800,-0.8632],[803,-0.865],[837,-0.9889],[887,-0.8495],[890,-0.8598],[891,-0.8607],[892,-0.8632],[895,-0.865],[927,-0.9878],[929,-0.9889]]
       }
       hips_ctrl.Location.Z {
        type=MovieSceneScriptingFloatChannel
        default=0
        thinned_from=752
        keys [[-2,0.0566],[-1,0.0558],[10,0.0533],[13,0.0521],[15,0.051],[16,0.0502],[17,0.0498],[19,0.0481],[20,0.0478],[21,0.0469],[22,0.0464],[23,0.0454],[24,0.0452],[26,0.0439],[28,0.0436],[31,0.0428],[44,0.0409],[45,0.0412],[47,0.0408],[48,0.041],[49,0.0408],[50,0.041],[51,0.0406],[57,0.0411],[60,0.0411],[61,0.0415],[67,0.0416],[68,0.0422],[70,0.0423],[71,0.0419],[75,0.0427],[76,0.0424],[77,0.0425],[78,0.0428],[79,0.0424],[80,0.0433],[81,0.0432],[82,0.0429],[83,0.0422],[84,0.0427],[85,0.0414],[86,0.0409],[87,0.0402],[88,0.0401],[90,0.0364],[91,0.0354],[92,0.0331],[94,0.0311],[95,0.0297],[96,0.0289],[97,0.0284],[98,0.0283],[99,0.0285],[101,0.0298],[102,0.0307],[104,0.0332],[107,0.0376],[108,0.0393],[110,0.0421],[111,0.0431],[112,0.0446],[114,0.0468],[115,0.0473],[116,0.0482],[117,0.0501],[118,0.051],[119,0.0509],[120,0.0524],[121,0.0542],[122,0.0545],[123,0.057],[124,0.0574],[125,0.0595],[126,0.0596],[127,0.0613],[128,0.0622],[129,0.0635],[130,0.0645],[131,0.0653],[132,0.0658],[133,0.0668],[135,0.0676],[136,0.0682],[137,0.0681],[138,0.0677],[139,0.0681],[142,0.0673],[143,0.0665],[145,0.0658],[146,0.0656],[147,0.066],[149,0.0649],[150,0.0641],[153,0.0629],[155,0.0626],[156,0.0617],[157,0.0612],[158,0.0613],[159,0.0609],[161,0.0603],[162,0.0604],[163,0.0602],[164,0.0603],[166,0.0601],[168,0.06],[170,0.0604],[171,0.0607],[174,0.0611],[175,0.0614],[176,0.0613],[177,0.0618],[178,0.0618],[179,0.0626],[180,0.0622],[185,0.0633],[186,0.0633],[187,0.0635],[188,0.0634],[189,0.064],[190,0.0637],[191,0.0647],[192,0.0639],[193,0.0647],[194,0.0641],[195,0.0649],[196,0.0639],[197,0.0647],[198,0.0638],[199,0.0644],[200,0.0645],[201,0.0642],[202,0.0641],[203,0.0646],[205,0.0641],[206,0.0646],[208,0.0645],[209,0.0648],[210,0.0645],[216,0.0648],[218,0.0643],[219,0.0643],[221,0.0637],[222,0.0632],[224,0.0616],[225,0.0612],[226,0.0601],[227,0.0594],[230,0.0564],[231,0.0558],[232,0.055],[233,0.0545],[234,0.0536],[235,0.053],[236,0.0523],[240,0.0481],[241,0.0474],[242,0.0461],[244,0.044],[245,0.0427],[246,0.0417],[247,0.0402],[251,0.036],[252,0.0354],[254,0.0336],[255,0.0333],[256,0.0328],[258,0.0316],[259,0.0312],[262,0.0296],[263,0.0294],[264,0.0289],[265,0.029],[268,0.0281],[269,0.0281],[270,0.0278],[274,0.0283],[275,0.0281],[279,0.0286],[280,0.0286],[281,0.0284],[283,0.0287],[284,0.0284],[285,0.0287],[286,0.0284],[287,0.0289],[291,0.0279],[292,0.0279],[293,0.0283],[294,0.0276],[295,0.0281],[296,0.0275],[297,0.0276],[298,0.0272],[299,0.0272],[300,0.0268],[302,0.0272],[307,0.0262],[309,0.0264],[323,0.0566],[324,0.0558],[335,0.0533],[338,0.0521],[340,0.051],[341,0.0502],[342,0.0498],[344,0.0481],[345,0.0478],[346,0.0469],[347,0.0464],[348,0.0454],[349,0.0452],[351,0.0439],[353,0.0436],[356,0.0428],[369,0.0409],[370,0.0412],[372,0.0408],[373,0.041],[374,0.0408],[375,0.041],[376,0.0406],[382,0.0411],[385,0.0411],[386,0.0415],[392,0.0416],[393,0.0422],[395,0.0423],[396,0.0419],[400,0.0427],[401,0.0424],[402,0.0425],[403,0.0428],[404,0.0424],[405,0.0433],[406,0.0432],[407,0.0429],[408,0.0422],[409,0.0427],[410,0.0414],[411,0.0409],[412,0.0402],[413,0.0401],[415,0.0364],[416,0.0354],[417,0.0331],[419,0.0311],[420,0.0297],[421,0.0289],[422,0.0284],[423,0.0283],[424,0.0285],[426,0.0298],[427,0.0307],[429,0.0332],[432,0.0376],[433,0.0393],[435,0.0421],[436,0.0431],[437,0.0446],[439,0.0468],[440,0.0473],[441,0.0482],[442,0.0501],[443,0.051],[444,0.0509],[445,0.0524],[446,0.0542],[447,0.0545],[448,0.057],[449,0.0574],[450,0.0595],[451,0.0596],[452,0.0613],[453,0.0622],[454,0.0635],[455,0.0645],[456,0.0653],[457,0.0658],[458,0.0668],[460,0.0676],[461,0.0682],[462,0.0681],[463,0.0677],[464,0.0681],[467,0.0673],[468,0.0665],[470,0.0658],[471,0.0656],[472,0.066],[474,0.0649],[475,0.0641],[478,0.0629],[480,0.0626],[481,0.0617],[482,0.0612],[483,0.0613],[484,0.0609],[486,0.0603],[487,0.0604],[488,0.0602],[489,0.0603],[491,0.0601],[493,0.06],[495,0.0604],[496,0.0607],[499,0.0611],[500,0.0614],[501,0.0613],[502,0.0618],[503,0.0618],[504,0.0626],[505,0.0622],[510,0.0633],[511,0.0633],[512,0.0635],[513,0.0634],[514,0.064],[515,0.0637],[516,0.0647],[517,0.0639],[518,0.0647],[519,0.0641],[520,0.0649],[521,0.0639],[522,0.0647],[523,0.0638],[524,0.0644],[525,0.0645],[526,0.0642],[527,0.0641],[528,0.0646],[530,0.0641],[531,0.0646],[533,0.0645],[534,0.0648],[535,0.0645],[541,0.0648],[543,0.0643],[544,0.0643],[546,0.0637],[547,0.0632],[549,0.0616],[550,0.0612],[551,0.0601],[552,0.0594],[555,0.0564],[556,0.0558],[557,0.055],[558,0.0545],[559,0.0536],[560,0.053],[561,0.0523],[565,0.0481],[566,0.0474],[567,0.0461],[569,0.044],[570,0.0427],[571,0.0417],[572,0.0402],[576,0.036],[577,0.0354],[579,0.0336],[580,0.0333],[581,0.0328],[583,0.0316],[584,0.0312],[587,0.0296],[588,0.0294],[589,0.0289],[590,0.029],[593,0.0281],[594,0.0281],[595,0.0278],[599,0.0283],[600,0.0281],[604,0.0286],[605,0.0286],[606,0.0284],[608,0.0287],[609,0.0284],[610,0.0287],[611,0.0284],[612,0.0289],[616,0.0279],[617,0.0279],[618,0.0283],[619,0.0276],[620,0.0281],[621,0.0275],[622,0.0276],[623,0.0272],[624,0.0272],[625,0.0268],[627,0.0272],[632,0.0262],[634,0.0264],[676,0.0566],[677,0.0558],[688,0.0533],[691,0.0521],[693,0.051],[694,0.0502],[695,0.0498],[697,0.0481],[698,0.0478],[699,0.0469],[700,0.0464],[701,0.0454],[702,0.0452],[704,0.0439],[706,0.0436],[709,0.0428],[722,0.0409],[723,0.0412],[725,0.0408],[726,0.041],[727,0.0408],[728,0.041],[729,0.0406],[735,0.0411],[738,0.0411],[739,0.0415],[745,0.0416],[746,0.0422],[748,0.0423],[749,0.0419],[753,0.0427],[754,0.0424],[757,0.0424],[759,0.0432],[770,0.0287],[771,0.0284],[772,0.0289],[776,0.0279],[777,0.0279],[778,0.0283],[779,0.0276],[780,0.0281],[781,0.0275],[782,0.0276],[783,0.0272],[784,0.0272],[785,0.0268],[787,0.0272],[792,0.0262],[794,0.0264],[797,0.026],[798,0.0256],[800,0.0254],[801,0.0251],[803,0.0249],[837,0.0566],[887,0.0262],[889,0.026],[890,0.0256],[892,0.0254],[893,0.0251],[895,0.0249],[927,0.0563],[929,0.0566]]
       }
       hips_ctrl.Rotation.X {
        type=MovieSceneScriptingFloatChannel
        default=0
        thinned_from=402
        keys [[-2,3.66],[1,3.6313],[3,3.5893],[5,3.5288],[9,3.3834],[17,3.1303],[19,3.0788],[21,3.0429],[23,3.0294],[25,3.0383],[28,3.0817],[31,3.1504],[33,3.2121],[34,3.2492],[35,3.2937],[37,3.4019],[39,3.5342],[41,3.6857],[45,4.03],[50,4.4747],[53,4.7075],[56,4.9043],[59,5.0622],[62,5.181],[65,5.2579],[68,5.2946],[71,5.2931],[74,5.2525],[76,5.1243],[83,3.81],[103,-3.7331],[114,-4.1489],[148,-0.7102],[163,0.1722],[177,0.9235],[182,1.2427],[184,1.3593],[185,1.4114],[186,1.4599],[187,1.5049],[188,1.5426],[189,1.5753],[190,1.6011],[191,1.6203],[192,1.632],[193,1.6372],[194,1.6368],[195,1.6301],[196,1.6192],[197,1.6035],[198,1.5828],[199,1.5575],[200,1.5278],[201,1.4937],[202,1.4553],[203,1.4131],[204,1.3669],[206,1.2647],[207,1.2078],[208,1.1435],[209,1.0736],[210,0.9968],[211,0.9148],[212,0.8275],[213,0.7342],[229,-2.7577],[233,-3.6543],[239,-4.5074],[263,-3.6992],[265,-3.6768],[267,-3.629],[269,-3.5617],[271,-3.4777],[273,-3.3781],[276,-3.2051],[279,-3.0567],[281,-2.9752],[283,-2.9084],[286,-2.838],[287,-2.8222],[289,-2.8106],[291,-2.8234],[292,-2.8378],[293,-2.8603],[295,-2.924],[297,-3.0055],[299,-3.1017],[302,-3.267],[313,-3.0045],[328,-1.0982],[359,1.9808],[360,2.0079],[362,2.0765],[364,2.1634],[366,2.2647],[369,2.4374],[375,2.8018],[378,2.9579],[380,3.0474],[382,3.1236],[384,3.1867],[387,3.2569],[390,3.2955],[393,3.3038],[396,3.2833],[399,3.2332],[401,3.1236],[408,2.0867],[428,-3.7331],[439,-4.1489],[473,-0.7102],[488,0.1722],[502,0.9235],[507,1.2427],[509,1.3593],[510,1.4114],[511,1.4599],[512,1.5049],[513,1.5426],[514,1.5753],[515,1.6011],[516,1.6203],[517,1.632],[518,1.6372],[519,1.6368],[520,1.6301],[521,1.6192],[522,1.6035],[523,1.5828],[524,1.5575],[525,1.5278],[526,1.4937],[527,1.4553],[528,1.4131],[529,1.3669],[531,1.2647],[532,1.2078],[533,1.1435],[534,1.0736],[535,0.9968],[536,0.9148],[537,0.8275],[538,0.7342],[554,-2.7577],[558,-3.6543],[564,-4.5074],[588,-3.6992],[590,-3.6768],[592,-3.629],[594,-3.5617],[596,-3.4777],[598,-3.3781],[601,-3.2051],[604,-3.0567],[606,-2.9752],[608,-2.9084],[611,-2.838],[612,-2.8222],[614,-2.8106],[616,-2.8234],[617,-2.8378],[618,-2.8603],[620,-2.924],[622,-3.0055],[624,-3.1017],[633,-3.6051],[634,-3.6522],[636,-3.6112],[676,-1.8223],[703,-0.5371],[704,-0.4539],[706,-0.2955],[707,-0.2201],[708,-0.1453],[709,-0.0756],[710,-0.009],[711,0.0546],[712,0.1151],[713,0.1742],[714,0.231],[716,0.3359],[717,0.3831],[718,0.4295],[719,0.4709],[720,0.5107],[721,0.5471],[722,0.5801],[723,0.6096],[724,0.6356],[725,0.6573],[726,0.6718],[727,0.6809],[728,0.6862],[729,0.6823],[730,0.6736],[731,0.6596],[732,0.6385],[733,0.6137],[734,0.5827],[735,0.5473],[736,0.5074],[737,0.464],[738,0.4171],[739,0.3665],[740,0.3152],[743,0.1489],[745,0.0366],[746,-0.0173],[747,-0.0705],[748,-0.1215],[749,-0.1694],[750,-0.2159],[751,-0.2587],[752,-0.2964],[753,-0.3319],[754,-0.3665],[755,-0.3968],[757,-0.4486],[759,-0.4829],[770,-0.5558],[771,-0.5467],[772,-0.5427],[773,-0.5475],[774,-0.5551],[775,-0.5711],[776,-0.5927],[777,-0.6197],[778,-0.6551],[779,-0.6988],[780,-0.7452],[781,-0.797],[782,-0.8541],[783,-0.9137],[784,-0.9786],[785,-1.0461],[786,-1.116],[789,-1.3344],[790,-1.4096],[793,-1.6242],[795,-1.7531],[797,-1.8647],[799,-1.9559],[801,-2.027],[803,-2.0842],[837,-0.4967],[887,-1.2659],[888,-1.324],[889,-1.3788],[890,-1.4266],[891,-1.4712],[892,-1.5091],[893,-1.5435],[894,-1.5745],[895,-1.6021],[927,3.6076],[929,3.66]]
       }
       hips_ctrl.Rotation.Y {
        type=MovieSceneScriptingFloatChannel
        default=0
        thinned_from=294
        keys [[-2,0.5304],[0,0.5291],[48,0.2605],[56,0.5959],[57,0.6802],[58,0.7688],[59,0.8617],[61,1.0578],[65,1.466],[67,1.6647],[69,1.8517],[70,1.9387],[71,2.0211],[72,2.0965],[73,2.1661],[74,2.2268],[75,2.2782],[76,2.3211],[77,2.3521],[78,2.3728],[80,2.3868],[82,2.3927],[83,2.4028],[84,2.4251],[85,2.4656],[86,2.524],[87,2.5948],[88,2.6886],[89,2.8029],[90,2.9371],[91,3.0802],[94,3.5433],[95,3.6894],[96,3.8217],[97,3.9337],[98,4.0343],[99,4.123],[100,4.1981],[101,4.2577],[102,4.3063],[103,4.3311],[104,4.3387],[105,4.3286],[106,4.3052],[107,4.2682],[108,4.2218],[110,4.104],[113,3.8855],[116,3.6602],[117,3.5971],[118,3.5417],[119,3.4969],[120,3.4638],[121,3.4417],[122,3.4272],[123,3.4222],[124,3.425],[127,3.4706],[130,3.5265],[132,3.5466],[134,3.5489],[136,3.531],[147,3.3244],[154,3.2393],[190,6.0904],[211,7.3654],[219,6.7186],[275,-2.6968],[285,-2.754],[291,-2.769],[294,-2.7565],[296,-2.7357],[298,-2.6955],[300,-2.6359],[302,-2.5574],[304,-2.4611],[308,-2.2407],[309,-2.1887],[323,0.5304],[325,0.5291],[373,0.2605],[381,0.5959],[382,0.6802],[383,0.7688],[384,0.8617],[386,1.0578],[390,1.466],[392,1.6647],[394,1.8517],[395,1.9387],[396,2.0211],[397,2.0965],[398,2.1661],[399,2.2268],[400,2.2782],[401,2.3211],[402,2.3521],[403,2.3728],[405,2.3868],[407,2.3927],[408,2.4028],[409,2.4251],[410,2.4656],[411,2.524],[412,2.5948],[413,2.6886],[414,2.8029],[415,2.9371],[416,3.0802],[419,3.5433],[420,3.6894],[421,3.8217],[422,3.9337],[423,4.0343],[424,4.123],[425,4.1981],[426,4.2577],[427,4.3063],[428,4.3311],[429,4.3387],[430,4.3286],[431,4.3052],[432,4.2682],[433,4.2218],[435,4.104],[438,3.8855],[441,3.6602],[442,3.5971],[443,3.5417],[444,3.4969],[445,3.4638],[446,3.4417],[447,3.4272],[448,3.4222],[449,3.425],[452,3.4706],[455,3.5265],[457,3.5466],[459,3.5489],[461,3.531],[472,3.3244],[479,3.2393],[515,6.0904],[536,7.3654],[544,6.7186],[600,-2.6968],[610,-2.754],[616,-2.769],[619,-2.7565],[621,-2.7357],[623,-2.6955],[625,-2.6359],[627,-2.5574],[629,-2.4611],[633,-2.2407],[634,-2.1887],[676,0.5304],[715,3.2807],[837,0.5304],[887,-2.1366],[889,-2.0462],[890,-2.0078],[891,-1.9747],[893,-1.9271],[895,-1.8905],[927,0.5091],[929,0.5304]]
       }
       hips_ctrl.Rotation.Z {
        type=MovieSceneScriptingFloatChannel
        default=0
        thinned_from=492
        keys [[-2,7.8157],[52,7.6856],[80,7.905],[82,7.9066],[102,0.7002],[124,8.0715],[136,8.196],[215,8.1008],[230,11.3375],[256,7.8208],[309,7.838],[377,7.6856],[405,7.905],[407,7.9066],[427,0.7002],[449,8.0715],[461,8.196],[540,8.1008],[555,11.3375],[581,7.8208],[634,7.838],[675,9.6027],[702,9.4297],[770,7.81],[803,7.8539],[837,7.8157],[895,7.8539],[929,7.8157]]
       }
       hips_ctrl.Scale.X {type=MovieSceneScriptingFloatChannel, default=1, const [1,752]}
       hips_ctrl.Scale.Y {type=MovieSceneScriptingFloatChannel, default=1, const [1,752]}
       hips_ctrl.Scale.Z {type=MovieSceneScriptingFloatChannel, default=1, const [1,752]}
       hips_tan_ctrl.Location.X {
        type=MovieSceneScriptingFloatChannel
        default=0
        thinned_from=751
        keys [[-2,-0.0581],[-1,-0.0381],[0,-0.0361],[1,-0.026],[2,-0.0166],[3,-0.0077],[4,0.0008],[5,0.0079],[6,0.0154],[7,0.0223],[8,0.0306],[9,0.0419],[10,0.0558],[11,0.0656],[12,0.0848],[13,0.1062],[14,0.1212],[15,0.1431],[16,0.1669],[17,0.1803],[18,0.2027],[19,0.2168],[21,0.2534],[22,0.2634],[23,0.2771],[24,0.2816],[25,0.2904],[26,0.2917],[27,0.2964],[28,0.3039],[29,0.3051],[30,0.3031],[32,0.3054],[34,0.3025],[35,0.3031],[38,0.2969],[39,0.2938],[45,0.2795],[46,0.278],[47,0.2778],[48,0.273],[51,0.2677],[52,0.2653],[53,0.265],[55,0.2605],[56,0.2592],[57,0.2537],[58,0.2531],[60,0.248],[61,0.2441],[63,0.2395],[64,0.2402],[65,0.234],[66,0.2315],[67,0.2297],[69,0.225],[70,0.222],[73,0.2159],[74,0.2145],[75,0.2083],[76,0.2097],[77,0.2083],[78,0.2065],[79,0.2059],[80,0.2038],[81,0.1991],[83,0.1954],[84,0.1962],[85,0.1943],[86,0.1877],[87,0.188],[88,0.1833],[89,0.1755],[90,0.1698],[91,0.1704],[92,0.1624],[93,0.1509],[94,0.1376],[95,0.1303],[96,0.1116],[97,0.0963],[98,0.0783],[99,0.0634],[100,0.0294],[101,0.0113],[102,-0.0192],[103,-0.0396],[104,-0.0712],[105,-0.0865],[106,-0.1185],[107,-0.143],[108,-0.1608],[109,-0.1837],[110,-0.2052],[111,-0.2244],[112,-0.2421],[113,-0.2585],[114,-0.2625],[115,-0.2867],[116,-0.2985],[117,-0.2968],[118,-0.3159],[119,-0.3108],[120,-0.3254],[121,-0.3269],[122,-0.3169],[123,-0.3269],[125,-0.3223],[126,-0.3186],[127,-0.3156],[128,-0.3118],[131,-0.3042],[134,-0.299],[136,-0.2975],[137,-0.2982],[139,-0.2973],[144,-0.2989],[150,-0.3021],[157,-0.3046],[158,-0.3057],[159,-0.3059],[165,-0.3135],[166,-0.3131],[171,-0.3173],[172,-0.3194],[175,-0.3203],[176,-0.322],[177,-0.3169],[178,-0.3188],[179,-0.318],[180,-0.3198],[181,-0.3156],[182,-0.3157],[183,-0.3171],[184,-0.3175],[186,-0.3171],[187,-0.3194],[188,-0.3138],[190,-0.3141],[191,-0.3159],[192,-0.3147],[194,-0.314],[195,-0.3126],[198,-0.3115],[199,-0.312],[200,-0.3118],[201,-0.3102],[202,-0.3108],[205,-0.309],[206,-0.3079],[207,-0.3096],[208,-0.3089],[209,-0.3044],[210,-0.3039],[212,-0.3066],[213,-0.3023],[214,-0.304],[215,-0.3066],[216,-0.3074],[217,-0.3038],[218,-0.3068],[219,-0.3116],[220,-0.3179],[221,-0.3267],[222,-0.3375],[224,-0.3642],[225,-0.3726],[227,-0.3993],[228,-0.4045],[229,-0.4122],[230,-0.4137],[231,-0.4142],[233,-0.4038],[234,-0.393],[235,-0.3877],[236,-0.3752],[237,-0.3687],[239,-0.3455],[240,-0.3326],[241,-0.3234],[242,-0.3095],[243,-0.3004],[245,-0.273],[246,-0.2646],[248,-0.2385],[249,-0.2297],[250,-0.2161],[253,-0.1845],[254,-0.173],[255,-0.1635],[256,-0.15],[257,-0.142],[258,-0.1298],[259,-0.1148],[260,-0.106],[261,-0.0945],[262,-0.0816],[263,-0.0728],[264,-0.0691],[265,-0.0635],[266,-0.0591],[267,-0.0562],[268,-0.0539],[269,-0.0492],[270,-0.0474],[273,-0.0409],[274,-0.039],[275,-0.0333],[277,-0.0297],[278,-0.0281],[279,-0.0261],[280,-0.0243],[281,-0.035],[282,-0.03],[283,-0.0285],[284,-0.0273],[285,-0.0258],[286,-0.0212],[287,-0.0201],[289,-0.0184],[291,-0.0173],[292,-0.0163],[293,-0.0161],[294,-0.0156],[295,-0.0155],[296,-0.0151],[297,-0.0103],[298,-0.0141],[299,-0.0135],[300,-0.0131],[302,-0.012],[303,-0.0126],[304,-0.0107],[305,-0.0129],[306,-0.0134],[307,-0.0116],[308,-0.0108],[309,-0.0138],[323,-0.0581],[324,-0.0381],[325,-0.0361],[326,-0.026],[327,-0.0166],[328,-0.0077],[329,0.0008],[330,0.0079],[331,0.0154],[332,0.0223],[333,0.0306],[334,0.0419],[335,0.0558],[336,0.0656],[337,0.0848],[338,0.1062],[339,0.1212],[340,0.1431],[341,0.1669],[342,0.1803],[343,0.2027],[344,0.2168],[346,0.2534],[347,0.2634],[348,0.2771],[349,0.2816],[350,0.2904],[351,0.2917],[352,0.2964],[353,0.3039],[354,0.3051],[355,0.3031],[357,0.3054],[359,0.3025],[360,0.3031],[363,0.2969],[364,0.2938],[370,0.2795],[371,0.278],[372,0.2778],[373,0.273],[376,0.2677],[377,0.2653],[378,0.265],[380,0.2605],[381,0.2592],[382,0.2537],[383,0.2531],[385,0.248],[386,0.2441],[388,0.2395],[389,0.2402],[390,0.234],[391,0.2315],[392,0.2297],[394,0.225],[395,0.222],[398,0.2159],[399,0.2145],[400,0.2083],[401,0.2097],[402,0.2083],[403,0.2065],[404,0.2059],[405,0.2038],[406,0.1991],[408,0.1954],[409,0.1962],[410,0.1943],[411,0.1877],[412,0.188],[413,0.1833],[414,0.1755],[415,0.1698],[416,0.1704],[417,0.1624],[418,0.1509],[419,0.1376],[420,0.1303],[421,0.1116],[422,0.0963],[423,0.0783],[424,0.0634],[425,0.0294],[426,0.0113],[427,-0.0192],[428,-0.0396],[429,-0.0712],[430,-0.0865],[431,-0.1185],[432,-0.143],[433,-0.1608],[434,-0.1837],[435,-0.2052],[436,-0.2244],[437,-0.2421],[438,-0.2585],[439,-0.2625],[440,-0.2867],[441,-0.2985],[442,-0.2968],[443,-0.3159],[444,-0.3108],[445,-0.3254],[446,-0.3269],[447,-0.3169],[448,-0.3269],[450,-0.3223],[451,-0.3186],[452,-0.3156],[453,-0.3118],[456,-0.3042],[459,-0.299],[461,-0.2975],[462,-0.2982],[464,-0.2973],[469,-0.2989],[475,-0.3021],[482,-0.3046],[483,-0.3057],[484,-0.3059],[490,-0.3135],[491,-0.3131],[496,-0.3173],[497,-0.3194],[500,-0.3203],[501,-0.322],[502,-0.3169],[503,-0.3188],[504,-0.318],[505,-0.3198],[506,-0.3156],[507,-0.3157],[508,-0.3171],[509,-0.3175],[511,-0.3171],[512,-0.3194],[513,-0.3138],[515,-0.3141],[516,-0.3159],[517,-0.3147],[519,-0.314],[520,-0.3126],[523,-0.3115],[524,-0.312],[525,-0.3118],[526,-0.3102],[527,-0.3108],[530,-0.309],[531,-0.3079],[532,-0.3096],[533,-0.3089],[534,-0.3044],[535,-0.3039],[537,-0.3066],[538,-0.3023],[539,-0.304],[540,-0.3066],[541,-0.3074],[542,-0.3038],[543,-0.3068],[544,-0.3116],[545,-0.3179],[546,-0.3267],[547,-0.3375],[549,-0.3642],[550,-0.3726],[552,-0.3993],[553,-0.4045],[554,-0.4122],[555,-0.4137],[556,-0.4142],[558,-0.4038],[559,-0.393],[560,-0.3877],[561,-0.3752],[562,-0.3687],[564,-0.3455],[565,-0.3326],[566,-0.3234],[567,-0.3095],[568,-0.3004],[570,-0.273],[571,-0.2646],[573,-0.2385],[574,-0.2297],[575,-0.2161],[578,-0.1845],[579,-0.173],[580,-0.1635],[581,-0.15],[582,-0.142],[583,-0.1298],[584,-0.1148],[585,-0.106],[586,-0.0945],[587,-0.0816],[588,-0.0728],[589,-0.0691],[590,-0.0635],[591,-0.0591],[592,-0.0562],[593,-0.0539],[594,-0.0492],[595,-0.0474],[598,-0.0409],[599,-0.039],[600,-0.0333],[602,-0.0297],[603,-0.0281],[604,-0.0261],[605,-0.0243],[606,-0.035],[607,-0.03],[608,-0.0285],[609,-0.0273],[610,-0.0258],[611,-0.0212],[612,-0.0201],[614,-0.0184],[616,-0.0173],[617,-0.0163],[618,-0.0161],[619,-0.0156],[620,-0.0155],[621,-0.0151],[622,-0.0103],[623,-0.0141],[624,-0.0135],[625,-0.0131],[627,-0.012],[628,-0.0126],[629,-0.0107],[630,-0.0129],[631,-0.0134],[632,-0.0116],[633,-0.0108],[634,-0.0138],[676,-0.0581],[677,-0.0381],[678,-0.0361],[679,-0.026],[680,-0.0166],[681,-0.0077],[682,0.0008],[683,0.0079],[684,0.0154],[685,0.0223],[686,0.0306],[687,0.0419],[688,0.0558],[689,0.0656],[690,0.0848],[691,0.1062],[692,0.1212],[693,0.1431],[694,0.1669],[695,0.1803],[696,0.2027],[697,0.2168],[699,0.2534],[700,0.2634],[701,0.2771],[702,0.2816],[703,0.2904],[704,0.2917],[705,0.2964],[706,0.3039],[707,0.3051],[708,0.3031],[710,0.3054],[712,0.3025],[713,0.3031],[716,0.2969],[717,0.2938],[723,0.2795],[724,0.278],[725,0.2778],[726,0.273],[729,0.2677],[730,0.2653],[731,0.265],[733,0.2605],[734,0.2592],[735,0.2537],[736,0.2531],[738,0.248],[739,0.2441],[741,0.2395],[742,0.2402],[743,0.234],[744,0.2315],[745,0.2297],[747,0.225],[748,0.222],[751,0.2159],[752,0.2145],[753,0.2083],[754,0.2097],[757,0.2059],[759,0.1991],[770,-0.0258],[771,-0.0212],[772,-0.0201],[774,-0.0184],[776,-0.0173],[777,-0.0163],[778,-0.0161],[779,-0.0156],[780,-0.0155],[781,-0.0151],[782,-0.0103],[783,-0.0141],[784,-0.0135],[785,-0.0131],[787,-0.012],[788,-0.0126],[789,-0.0107],[790,-0.0129],[791,-0.0134],[792,-0.0116],[793,-0.0108],[794,-0.0138],[795,-0.0128],[796,-0.0127],[797,-0.0115],[798,-0.0112],[799,-0.0067],[800,-0.0072],[801,-0.0054],[802,-0.0042],[803,-0.005],[837,-0.0581],[887,-0.0128],[888,-0.0127],[889,-0.0115],[890,-0.0112],[891,-0.0067],[892,-0.0072],[893,-0.0054],[894,-0.0042],[895,-0.005],[929,-0.0581]]
       }
       hips_tan_ctrl.Location.Y {
        type=MovieSceneScriptingFloatChannel
        default=0.008
        thinned_from=751
        keys [[-2,-0.7297],[-1,-0.734],[0,-0.7357],[1,-0.7443],[5,-0.7487],[9,-0.7478],[10,-0.753],[12,-0.7485],[13,-0.7511],[18,-0.7322],[20,-0.7228],[21,-0.7169],[22,-0.7132],[23,-0.7052],[24,-0.701],[28,-0.6792],[29,-0.6721],[30,-0.6713],[31,-0.6651],[32,-0.6613],[33,-0.6544],[34,-0.6587],[35,-0.6524],[36,-0.6498],[37,-0.6439],[38,-0.6419],[39,-0.6437],[40,-0.6421],[41,-0.6375],[42,-0.6364],[43,-0.6319],[44,-0.6384],[45,-0.6378],[46,-0.6341],[48,-0.6359],[49,-0.6351],[50,-0.6359],[54,-0.6465],[58,-0.6634],[64,-0.693],[65,-0.6993],[69,-0.7176],[70,-0.7258],[71,-0.7277],[74,-0.7383],[75,-0.7446],[76,-0.7454],[82,-0.7578],[83,-0.7636],[84,-0.7674],[86,-0.7824],[87,-0.7932],[89,-0.8195],[90,-0.8312],[92,-0.8618],[97,-0.921],[98,-0.9271],[99,-0.9363],[101,-0.9511],[102,-0.9536],[103,-0.9593],[104,-0.9608],[105,-0.9645],[108,-0.961],[110,-0.9546],[111,-0.9497],[112,-0.9415],[113,-0.9389],[115,-0.9236],[116,-0.9121],[119,-0.8671],[120,-0.8462],[121,-0.8313],[124,-0.77],[125,-0.7462],[126,-0.7309],[128,-0.6981],[130,-0.6715],[133,-0.6371],[134,-0.6343],[135,-0.6262],[136,-0.6248],[138,-0.6274],[139,-0.6351],[141,-0.6403],[142,-0.6475],[143,-0.6527],[144,-0.6626],[145,-0.6666],[146,-0.6767],[147,-0.681],[148,-0.6916],[149,-0.6944],[150,-0.705],[151,-0.709],[152,-0.7189],[153,-0.7225],[154,-0.7319],[155,-0.7345],[156,-0.7392],[157,-0.7477],[159,-0.7531],[160,-0.7605],[161,-0.7619],[162,-0.7593],[163,-0.7608],[164,-0.7661],[166,-0.7683],[167,-0.7643],[168,-0.7696],[171,-0.7689],[172,-0.7638],[174,-0.7612],[175,-0.7565],[176,-0.7551],[177,-0.7492],[179,-0.744],[180,-0.7371],[183,-0.7281],[184,-0.7192],[185,-0.7148],[186,-0.7119],[187,-0.7076],[188,-0.6993],[189,-0.7004],[190,-0.6921],[192,-0.6862],[193,-0.6807],[194,-0.6808],[195,-0.6758],[196,-0.6745],[197,-0.6698],[198,-0.6706],[199,-0.6662],[200,-0.6674],[202,-0.6603],[203,-0.6619],[205,-0.6612],[206,-0.658],[209,-0.6573],[211,-0.6603],[214,-0.6615],[215,-0.6752],[216,-0.6742],[217,-0.6755],[218,-0.6731],[219,-0.6825],[220,-0.6798],[221,-0.6739],[222,-0.6831],[224,-0.6895],[225,-0.6979],[226,-0.7021],[228,-0.7202],[229,-0.7321],[230,-0.748],[231,-0.7542],[232,-0.7645],[235,-0.8037],[236,-0.8178],[237,-0.8352],[238,-0.8422],[239,-0.8611],[241,-0.8895],[242,-0.9083],[244,-0.9372],[245,-0.956],[246,-0.9711],[247,-0.9891],[248,-1.0026],[250,-1.0346],[251,-1.0482],[252,-1.0594],[255,-1.099],[258,-1.1317],[261,-1.1578],[264,-1.1742],[266,-1.1828],[267,-1.1821],[269,-1.1875],[270,-1.1851],[272,-1.1846],[273,-1.189],[276,-1.1818],[277,-1.1631],[278,-1.1667],[282,-1.1627],[283,-1.168],[289,-1.1703],[291,-1.1766],[293,-1.1806],[294,-1.1863],[295,-1.1883],[296,-1.1948],[297,-1.1975],[299,-1.21],[305,-1.2417],[309,-1.2626],[323,-0.7297],[324,-0.734],[325,-0.7357],[326,-0.7443],[330,-0.7487],[334,-0.7478],[335,-0.753],[337,-0.7485],[338,-0.7511],[343,-0.7322],[345,-0.7228],[346,-0.7169],[347,-0.7132],[348,-0.7052],[349,-0.701],[353,-0.6792],[354,-0.6721],[355,-0.6713],[356,-0.6651],[357,-0.6613],[358,-0.6544],[359,-0.6587],[360,-0.6524],[361,-0.6498],[362,-0.6439],[363,-0.6419],[364,-0.6437],[365,-0.6421],[366,-0.6375],[367,-0.6364],[368,-0.6319],[369,-0.6384],[370,-0.6378],[371,-0.6341],[373,-0.6359],[374,-0.6351],[375,-0.6359],[379,-0.6465],[383,-0.6634],[389,-0.693],[390,-0.6993],[394,-0.7176],[395,-0.7258],[396,-0.7277],[399,-0.7383],[400,-0.7446],[401,-0.7454],[407,-0.7578],[408,-0.7636],[409,-0.7674],[411,-0.7824],[412,-0.7932],[414,-0.8195],[415,-0.8312],[417,-0.8618],[422,-0.921],[423,-0.9271],[424,-0.9363],[426,-0.9511],[427,-0.9536],[428,-0.9593],[429,-0.9608],[430,-0.9645],[433,-0.961],[435,-0.9546],[436,-0.9497],[437,-0.9415],[438,-0.9389],[440,-0.9236],[441,-0.9121],[444,-0.8671],[445,-0.8462],[446,-0.8313],[449,-0.77],[450,-0.7462],[451,-0.7309],[453,-0.6981],[455,-0.6715],[458,-0.6371],[459,-0.6343],[460,-0.6262],[461,-0.6248],[463,-0.6274],[464,-0.6351],[466,-0.6403],[467,-0.6475],[468,-0.6527],[469,-0.6626],[470,-0.6666],[471,-0.6767],[472,-0.681],[473,-0.6916],[474,-0.6944],[475,-0.705],[476,-0.709],[477,-0.7189],[478,-0.7225],[479,-0.7319],[480,-0.7345],[481,-0.7392],[482,-0.7477],[484,-0.7531],[485,-0.7605],[486,-0.7619],[487,-0.7593],[488,-0.7608],[489,-0.7661],[491,-0.7683],[492,-0.7643],[493,-0.7696],[496,-0.7689],[497,-0.7638],[499,-0.7612],[500,-0.7565],[501,-0.7551],[502,-0.7492],[504,-0.744],[505,-0.7371],[508,-0.7281],[509,-0.7192],[510,-0.7148],[511,-0.7119],[512,-0.7076],[513,-0.6993],[514,-0.7004],[515,-0.6921],[517,-0.6862],[518,-0.6807],[519,-0.6808],[520,-0.6758],[521,-0.6745],[522,-0.6698],[523,-0.6706],[524,-0.6662],[525,-0.6674],[527,-0.6603],[528,-0.6619],[530,-0.6612],[531,-0.658],[534,-0.6573],[536,-0.6603],[539,-0.6615],[540,-0.6752],[541,-0.6742],[542,-0.6755],[543,-0.6731],[544,-0.6825],[545,-0.6798],[546,-0.6739],[547,-0.6831],[549,-0.6895],[550,-0.6979],[551,-0.7021],[553,-0.7202],[554,-0.7321],[555,-0.748],[556,-0.7542],[557,-0.7645],[560,-0.8037],[561,-0.8178],[562,-0.8352],[563,-0.8422],[564,-0.8611],[566,-0.8895],[567,-0.9083],[569,-0.9372],[570,-0.956],[571,-0.9711],[572,-0.9891],[573,-1.0026],[575,-1.0346],[576,-1.0482],[577,-1.0594],[580,-1.099],[583,-1.1317],[586,-1.1578],[589,-1.1742],[591,-1.1828],[592,-1.1821],[594,-1.1875],[595,-1.1851],[597,-1.1846],[598,-1.189],[601,-1.1818],[602,-1.1631],[603,-1.1667],[607,-1.1627],[608,-1.168],[614,-1.1703],[616,-1.1766],[618,-1.1806],[619,-1.1863],[620,-1.1883],[621,-1.1948],[622,-1.1975],[624,-1.21],[630,-1.2417],[634,-1.2626],[676,-0.7297],[677,-0.734],[678,-0.7357],[679,-0.7443],[683,-0.7487],[687,-0.7478],[688,-0.753],[690,-0.7485],[691,-0.7511],[696,-0.7322],[698,-0.7228],[699,-0.7169],[700,-0.7132],[701,-0.7052],[702,-0.701],[706,-0.6792],[707,-0.6721],[708,-0.6713],[709,-0.6651],[710,-0.6613],[711,-0.6544],[712,-0.6587],[713,-0.6524],[714,-0.6498],[715,-0.6439],[716,-0.6419],[717,-0.6437],[718,-0.6421],[719,-0.6375],[720,-0.6364],[721,-0.6319],[722,-0.6384],[723,-0.6378],[724,-0.6341],[726,-0.6359],[727,-0.6351],[728,-0.6359],[732,-0.6465],[736,-0.6634],[742,-0.693],[743,-0.6993],[747,-0.7176],[748,-0.7258],[749,-0.7277],[752,-0.7383],[753,-0.7446],[754,-0.7454],[759,-0.7555],[770,-1.1677],[774,-1.1703],[776,-1.1766],[778,-1.1806],[779,-1.1863],[780,-1.1883],[781,-1.1948],[782,-1.1975],[784,-1.21],[790,-1.2417],[800,-1.2875],[803,-1.2963],[837,-0.7297],[887,-1.2677],[895,-1.2963],[929,-0.7297]]
       }
       hips_tan_ctrl.Location.Z {
        type=MovieSceneScriptingFloatChannel
        default=-0.0073
        thinned_from=751
        keys [[-2,-0.0348],[-1,-0.034],[0,-0.0338],[1,-0.0338],[2,-0.0341],[4,-0.034],[5,-0.0341],[6,-0.0338],[7,-0.034],[8,-0.0343],[9,-0.0305],[11,-0.0313],[12,-0.0275],[13,-0.0283],[14,-0.0247],[15,-0.0241],[16,-0.0206],[17,-0.0195],[18,-0.0144],[19,-0.0117],[20,-0.0097],[21,-0.0058],[22,-0.0045],[23,-0.0006],[24,0.0006],[25,0.0045],[26,0.0091],[27,0.0089],[29,0.0108],[30,0.014],[32,0.0139],[33,0.015],[34,0.0159],[35,0.0157],[36,0.0166],[37,0.0167],[38,0.0175],[39,0.0175],[40,0.0161],[43,0.0161],[44,0.016],[45,0.0151],[46,0.0151],[47,0.0161],[48,0.013],[50,0.0132],[51,0.012],[52,0.0126],[54,0.0127],[55,0.01],[56,0.01],[57,0.0091],[58,0.0092],[60,0.0037],[61,0.0057],[62,0.0031],[63,0.0032],[64,0.0027],[65,-0.0011],[67,-0.002],[68,-0.0019],[69,-0.0057],[70,-0.0062],[71,-0.0061],[72,-0.0066],[73,-0.0066],[74,-0.0071],[75,-0.0109],[76,-0.0109],[77,-0.0114],[78,-0.0114],[79,-0.01],[80,-0.01],[81,-0.0148],[82,-0.0136],[83,-0.014],[84,-0.0127],[85,-0.0161],[86,-0.015],[87,-0.018],[88,-0.0165],[89,-0.0194],[90,-0.0177],[91,-0.0205],[92,-0.0218],[93,-0.0202],[94,-0.022],[95,-0.0234],[96,-0.0234],[97,-0.0275],[98,-0.028],[99,-0.0363],[100,-0.0374],[101,-0.0377],[102,-0.0467],[103,-0.0469],[104,-0.0515],[105,-0.0517],[107,-0.0604],[108,-0.0561],[109,-0.0598],[110,-0.0596],[111,-0.0601],[112,-0.0592],[114,-0.058],[115,-0.0618],[116,-0.0529],[117,-0.0529],[118,-0.0487],[119,-0.0486],[120,-0.0444],[121,-0.0412],[122,-0.0414],[123,-0.0373],[124,-0.0339],[125,-0.0342],[126,-0.0311],[127,-0.0298],[131,-0.0228],[133,-0.019],[134,-0.0192],[135,-0.015],[136,-0.0175],[138,-0.0178],[139,-0.0162],[140,-0.0203],[141,-0.0205],[142,-0.0187],[143,-0.0203],[144,-0.0234],[145,-0.0226],[146,-0.0246],[147,-0.0237],[148,-0.0268],[149,-0.0258],[150,-0.0289],[151,-0.028],[152,-0.0311],[153,-0.0301],[154,-0.0333],[155,-0.0321],[156,-0.0322],[157,-0.0353],[158,-0.0345],[159,-0.0355],[160,-0.0345],[161,-0.0363],[162,-0.0354],[163,-0.0353],[164,-0.0373],[165,-0.0373],[166,-0.0363],[168,-0.0367],[169,-0.0349],[170,-0.0361],[172,-0.0365],[173,-0.0348],[174,-0.036],[175,-0.0354],[176,-0.0331],[177,-0.0332],[178,-0.0308],[179,-0.0345],[180,-0.0343],[181,-0.0322],[182,-0.0318],[183,-0.0294],[184,-0.0324],[185,-0.0298],[186,-0.0294],[187,-0.0268],[188,-0.0266],[189,-0.03],[190,-0.0272],[191,-0.0265],[192,-0.0262],[193,-0.0264],[194,-0.0255],[195,-0.0261],[196,-0.025],[197,-0.0255],[198,-0.0242],[199,-0.024],[200,-0.0245],[203,-0.024],[204,-0.0246],[205,-0.0234],[207,-0.0221],[208,-0.0244],[210,-0.024],[211,-0.0233],[213,-0.0228],[214,-0.0222],[215,-0.0248],[216,-0.0242],[218,-0.0238],[220,-0.0227],[221,-0.0219],[222,-0.0234],[223,-0.0216],[224,-0.026],[225,-0.0223],[226,-0.021],[227,-0.0259],[228,-0.022],[229,-0.0214],[230,-0.0263],[231,-0.0233],[232,-0.0291],[233,-0.0266],[234,-0.0324],[235,-0.0302],[236,-0.0337],[237,-0.034],[238,-0.037],[239,-0.0369],[240,-0.0438],[241,-0.0411],[242,-0.048],[243,-0.0451],[244,-0.0517],[245,-0.0488],[246,-0.0537],[247,-0.0522],[248,-0.049],[249,-0.0556],[250,-0.0526],[251,-0.0592],[252,-0.0564],[253,-0.0631],[254,-0.0604],[255,-0.0612],[256,-0.0667],[258,-0.0631],[259,-0.0686],[260,-0.0668],[261,-0.0665],[262,-0.0642],[263,-0.0712],[264,-0.069],[265,-0.0686],[268,-0.0668],[270,-0.0658],[272,-0.0655],[273,-0.0648],[274,-0.0665],[275,-0.0668],[276,-0.0668],[277,-0.0581],[278,-0.0592],[279,-0.0593],[280,-0.0592],[282,-0.0594],[283,-0.0585],[284,-0.0585],[285,-0.0601],[286,-0.0601],[287,-0.0608],[288,-0.0608],[291,-0.0604],[292,-0.0624],[293,-0.061],[294,-0.0608],[295,-0.0616],[296,-0.0614],[297,-0.0622],[298,-0.0621],[299,-0.0628],[300,-0.0647],[301,-0.0633],[302,-0.0642],[303,-0.0644],[305,-0.0652],[306,-0.0663],[307,-0.0662],[308,-0.0663],[309,-0.0668],[323,-0.0348],[324,-0.034],[325,-0.0338],[326,-0.0338],[327,-0.0341],[329,-0.034],[330,-0.0341],[331,-0.0338],[332,-0.034],[333,-0.0343],[334,-0.0305],[336,-0.0313],[337,-0.0275],[338,-0.0283],[339,-0.0247],[340,-0.0241],[341,-0.0206],[342,-0.0195],[343,-0.0144],[344,-0.0117],[345,-0.0097],[346,-0.0058],[347,-0.0045],[348,-0.0006],[349,0.0006],[350,0.0045],[351,0.0091],[352,0.0089],[354,0.0108],[355,0.014],[357,0.0139],[358,0.015],[359,0.0159],[360,0.0157],[361,0.0166],[362,0.0167],[363,0.0175],[364,0.0175],[365,0.0161],[368,0.0161],[369,0.016],[370,0.0151],[371,0.0151],[372,0.0161],[373,0.013],[375,0.0132],[376,0.012],[377,0.0126],[379,0.0127],[380,0.01],[381,0.01],[382,0.0091],[383,0.0092],[385,0.0037],[386,0.0057],[387,0.0031],[388,0.0032],[389,0.0027],[390,-0.0011],[392,-0.002],[393,-0.0019],[394,-0.0057],[395,-0.0062],[396,-0.0061],[397,-0.0066],[398,-0.0066],[399,-0.0071],[400,-0.0109],[401,-0.0109],[402,-0.0114],[403,-0.0114],[404,-0.01],[405,-0.01],[406,-0.0148],[407,-0.0136],[408,-0.014],[409,-0.0127],[410,-0.0161],[411,-0.015],[412,-0.018],[413,-0.0165],[414,-0.0194],[415,-0.0177],[416,-0.0205],[417,-0.0218],[418,-0.0202],[419,-0.022],[420,-0.0234],[421,-0.0234],[422,-0.0275],[423,-0.028],[424,-0.0363],[425,-0.0374],[426,-0.0377],[427,-0.0467],[428,-0.0469],[429,-0.0515],[430,-0.0517],[432,-0.0604],[433,-0.0561],[434,-0.0598],[435,-0.0596],[436,-0.0601],[437,-0.0592],[439,-0.058],[440,-0.0618],[441,-0.0529],[442,-0.0529],[443,-0.0487],[444,-0.0486],[445,-0.0444],[446,-0.0412],[447,-0.0414],[448,-0.0373],[449,-0.0339],[450,-0.0342],[451,-0.0311],[452,-0.0298],[456,-0.0228],[458,-0.019],[459,-0.0192],[460,-0.015],[461,-0.0175],[463,-0.0178],[464,-0.0162],[465,-0.0203],[466,-0.0205],[467,-0.0187],[468,-0.0203],[469,-0.0234],[470,-0.0226],[471,-0.0246],[472,-0.0237],[473,-0.0268],[474,-0.0258],[475,-0.0289],[476,-0.028],[477,-0.0311],[478,-0.0301],[479,-0.0333],[480,-0.0321],[481,-0.0322],[482,-0.0353],[483,-0.0345],[484,-0.0355],[485,-0.0345],[486,-0.0363],[487,-0.0354],[488,-0.0353],[489,-0.0373],[490,-0.0373],[491,-0.0363],[493,-0.0367],[494,-0.0349],[495,-0.0361],[497,-0.0365],[498,-0.0348],[499,-0.036],[500,-0.0354],[501,-0.0331],[502,-0.0332],[503,-0.0308],[504,-0.0345],[505,-0.0343],[506,-0.0322],[507,-0.0318],[508,-0.0294],[509,-0.0324],[510,-0.0298],[511,-0.0294],[512,-0.0268],[513,-0.0266],[514,-0.03],[515,-0.0272],[516,-0.0265],[517,-0.0262],[518,-0.0264],[519,-0.0255],[520,-0.0261],[521,-0.025],[522,-0.0255],[523,-0.0242],[524,-0.024],[525,-0.0245],[528,-0.024],[529,-0.0246],[530,-0.0234],[532,-0.0221],[533,-0.0244],[535,-0.024],[536,-0.0233],[538,-0.0228],[539,-0.0222],[540,-0.0248],[541,-0.0242],[543,-0.0238],[545,-0.0227],[546,-0.0219],[547,-0.0234],[548,-0.0216],[549,-0.026],[550,-0.0223],[551,-0.021],[552,-0.0259],[553,-0.022],[554,-0.0214],[555,-0.0263],[556,-0.0233],[557,-0.0291],[558,-0.0266],[559,-0.0324],[560,-0.0302],[561,-0.0337],[562,-0.034],[563,-0.037],[564,-0.0369],[565,-0.0438],[566,-0.0411],[567,-0.048],[568,-0.0451],[569,-0.0517],[570,-0.0488],[571,-0.0537],[572,-0.0522],[573,-0.049],[574,-0.0556],[575,-0.0526],[576,-0.0592],[577,-0.0564],[578,-0.0631],[579,-0.0604],[580,-0.0612],[581,-0.0667],[583,-0.0631],[584,-0.0686],[585,-0.0668],[586,-0.0665],[587,-0.0642],[588,-0.0712],[589,-0.069],[590,-0.0686],[593,-0.0668],[595,-0.0658],[597,-0.0655],[598,-0.0648],[599,-0.0665],[600,-0.0668],[601,-0.0668],[602,-0.0581],[603,-0.0592],[604,-0.0593],[605,-0.0592],[607,-0.0594],[608,-0.0585],[609,-0.0585],[610,-0.0601],[611,-0.0601],[612,-0.0608],[613,-0.0608],[616,-0.0604],[617,-0.0624],[618,-0.061],[619,-0.0608],[620,-0.0616],[621,-0.0614],[622,-0.0622],[623,-0.0621],[624,-0.0628],[625,-0.0647],[626,-0.0633],[627,-0.0642],[628,-0.0644],[630,-0.0652],[631,-0.0663],[632,-0.0662],[633,-0.0663],[634,-0.0668],[677,-0.034],[678,-0.0338],[679,-0.0338],[680,-0.0341],[682,-0.034],[683,-0.0341],[684,-0.0338],[685,-0.034],[686,-0.0343],[687,-0.0305],[689,-0.0313],[690,-0.0275],[691,-0.0283],[692,-0.0247],[693,-0.0241],[694,-0.0206],[695,-0.0195],[696,-0.0144],[697,-0.0117],[698,-0.0097],[699,-0.0058],[700,-0.0045],[701,-0.0006],[702,0.0006],[703,0.0045],[704,0.0091],[705,0.0089],[707,0.0108],[708,0.014],[710,0.0139],[711,0.015],[712,0.0159],[713,0.0157],[714,0.0166],[715,0.0167],[716,0.0175],[717,0.0175],[718,0.0161],[721,0.0161],[722,0.016],[723,0.0151],[724,0.0151],[725,0.0161],[726,0.013],[728,0.0132],[729,0.012],[730,0.0126],[732,0.0127],[733,0.01],[734,0.01],[735,0.0091],[736,0.0092],[738,0.0037],[739,0.0057],[740,0.0031],[741,0.0032],[742,0.0027],[743,-0.0011],[745,-0.002],[746,-0.0019],[747,-0.0057],[748,-0.0062],[749,-0.0061],[750,-0.0066],[751,-0.0066],[752,-0.0071],[753,-0.0109],[754,-0.0109],[755,-0.0114],[757,-0.01],[759,-0.0148],[770,-0.0601],[771,-0.0601],[772,-0.0608],[773,-0.0608],[776,-0.0604],[777,-0.0624],[778,-0.061],[779,-0.0608],[780,-0.0616],[781,-0.0614],[782,-0.0622],[783,-0.0621],[784,-0.0628],[785,-0.0647],[786,-0.0633],[787,-0.0642],[788,-0.0644],[790,-0.0652],[791,-0.0663],[792,-0.0662],[793,-0.0663],[794,-0.0668],[795,-0.0669],[796,-0.0674],[799,-0.0678],[800,-0.0678],[803,-0.0684],[837,-0.0348],[888,-0.0674],[891,-0.0678],[892,-0.0678],[895,-0.0684],[929,-0.0348]]
       }
       hips_tan_ctrl.Rotation.X {type=MovieSceneScriptingFloatChannel, default=0, const [0,751]}
       hips_tan_ctrl.Rotation.Y {type=MovieSceneScriptingFloatChannel, default=0, const [0,751]}
       hips_tan_ctrl.Rotation.Z {type=MovieSceneScriptingFloatChannel, default=0, const [0,751]}
       hips_tan_ctrl.Scale.X {type=MovieSceneScriptingFloatChannel, default=1, const [1,751]}
       hips_tan_ctrl.Scale.Y {type=MovieSceneScriptingFloatChannel, default=1, const [1,751]}
       hips_tan_ctrl.Scale.Z {type=MovieSceneScriptingFloatChannel, default=1, const [1,751]}
       spine_01_ctrl.Location.X {
        type=MovieSceneScriptingFloatChannel
        default=0
        thinned_from=741
        keys [[-2,-0.0419],[3,-0.0461],[10,-0.0461],[11,-0.0504],[13,-0.0504],[14,-0.0546],[15,-0.0556],[16,-0.0575],[17,-0.0614],[19,-0.0652],[20,-0.069],[26,-0.0806],[27,-0.0806],[28,-0.0825],[29,-0.0825],[30,-0.0844],[31,-0.084],[34,-0.085],[35,-0.085],[36,-0.0853],[38,-0.0853],[39,-0.0856],[45,-0.0856],[47,-0.0851],[52,-0.0851],[53,-0.0845],[55,-0.0845],[56,-0.0839],[57,-0.0839],[58,-0.0833],[60,-0.0833],[61,-0.0827],[62,-0.0827],[63,-0.083],[64,-0.0813],[69,-0.0813],[70,-0.0795],[75,-0.0795],[76,-0.0778],[78,-0.0778],[79,-0.0773],[85,-0.0773],[86,-0.0812],[87,-0.0812],[90,-0.0927],[91,-0.0927],[94,-0.1043],[95,-0.1052],[96,-0.1081],[98,-0.1081],[99,-0.1052],[100,-0.1052],[104,-0.0939],[105,-0.0882],[106,-0.0854],[107,-0.0797],[108,-0.0769],[110,-0.0655],[111,-0.0617],[114,-0.0445],[115,-0.0445],[117,-0.033],[118,-0.033],[119,-0.0273],[121,-0.0273],[122,-0.0216],[126,-0.0216],[127,-0.0221],[128,-0.0221],[130,-0.0232],[131,-0.0232],[132,-0.0237],[136,-0.0237],[137,-0.0243],[141,-0.0243],[142,-0.0248],[144,-0.0247],[145,-0.0249],[148,-0.0249],[149,-0.0252],[153,-0.0252],[154,-0.0255],[157,-0.0255],[158,-0.0257],[159,-0.0253],[160,-0.0263],[171,-0.0263],[172,-0.0272],[174,-0.0272],[175,-0.0277],[176,-0.0288],[177,-0.0288],[178,-0.0299],[179,-0.0299],[181,-0.032],[182,-0.032],[183,-0.0331],[186,-0.0331],[187,-0.0342],[190,-0.0342],[191,-0.0347],[200,-0.0347],[201,-0.0341],[202,-0.0341],[203,-0.0336],[204,-0.0336],[206,-0.0325],[208,-0.0324],[209,-0.0293],[212,-0.0293],[213,-0.0261],[216,-0.0261],[217,-0.023],[224,-0.023],[225,-0.0204],[227,-0.0204],[228,-0.0178],[229,-0.0178],[230,-0.0153],[231,-0.0153],[232,-0.0127],[233,-0.0127],[234,-0.0101],[235,-0.0101],[236,-0.0076],[237,-0.0076],[238,-0.005],[239,-0.0036],[240,-0.0012],[241,-0.0012],[242,0.0011],[243,0.0011],[245,0.0059],[246,0.0059],[248,0.0106],[249,0.0106],[250,0.013],[254,0.013],[255,0.0115],[256,0.0115],[257,0.0082],[259,0.0082],[260,0.0049],[263,0.0049],[264,0.0016],[270,0.0016],[271,0.0026],[280,0.0026],[281,-0.0041],[286,-0.0041],[287,-0.0036],[288,-0.0036],[289,-0.006],[292,-0.006],[293,-0.0085],[294,-0.0085],[295,-0.0109],[297,-0.0109],[299,-0.0157],[300,-0.0157],[302,-0.0206],[303,-0.0216],[304,-0.0231],[305,-0.0262],[308,-0.0308],[309,-0.0339],[323,-0.0419],[328,-0.0461],[335,-0.0461],[336,-0.0504],[338,-0.0504],[339,-0.0546],[340,-0.0556],[341,-0.0575],[342,-0.0614],[344,-0.0652],[345,-0.069],[351,-0.0806],[352,-0.0806],[353,-0.0825],[354,-0.0825],[355,-0.0844],[356,-0.084],[359,-0.085],[360,-0.085],[361,-0.0853],[363,-0.0853],[364,-0.0856],[370,-0.0856],[372,-0.0851],[377,-0.0851],[378,-0.0845],[380,-0.0845],[381,-0.0839],[382,-0.0839],[383,-0.0833],[385,-0.0833],[386,-0.0827],[387,-0.0827],[388,-0.083],[389,-0.0813],[394,-0.0813],[395,-0.0795],[400,-0.0795],[401,-0.0778],[403,-0.0778],[404,-0.0773],[410,-0.0773],[411,-0.0812],[412,-0.0812],[415,-0.0927],[416,-0.0927],[419,-0.1043],[420,-0.1052],[421,-0.1081],[423,-0.1081],[424,-0.1052],[425,-0.1052],[429,-0.0939],[430,-0.0882],[431,-0.0854],[432,-0.0797],[433,-0.0769],[435,-0.0655],[436,-0.0617],[439,-0.0445],[440,-0.0445],[442,-0.033],[443,-0.033],[444,-0.0273],[446,-0.0273],[447,-0.0216],[451,-0.0216],[452,-0.0221],[453,-0.0221],[455,-0.0232],[456,-0.0232],[457,-0.0237],[461,-0.0237],[462,-0.0243],[466,-0.0243],[467,-0.0248],[469,-0.0247],[470,-0.0249],[473,-0.0249],[474,-0.0252],[478,-0.0252],[479,-0.0255],[482,-0.0255],[483,-0.0257],[484,-0.0253],[485,-0.0263],[496,-0.0263],[497,-0.0272],[499,-0.0272],[500,-0.0277],[501,-0.0288],[502,-0.0288],[503,-0.0299],[504,-0.0299],[506,-0.032],[507,-0.032],[508,-0.0331],[511,-0.0331],[512,-0.0342],[515,-0.0342],[516,-0.0347],[525,-0.0347],[526,-0.0341],[527,-0.0341],[528,-0.0336],[529,-0.0336],[531,-0.0325],[533,-0.0324],[534,-0.0293],[537,-0.0293],[538,-0.0261],[541,-0.0261],[542,-0.023],[549,-0.023],[550,-0.0204],[552,-0.0204],[553,-0.0178],[554,-0.0178],[555,-0.0153],[556,-0.0153],[557,-0.0127],[558,-0.0127],[559,-0.0101],[560,-0.0101],[561,-0.0076],[562,-0.0076],[563,-0.005],[564,-0.0036],[565,-0.0012],[566,-0.0012],[567,0.0011],[568,0.0011],[570,0.0059],[571,0.0059],[573,0.0106],[574,0.0106],[575,0.013],[579,0.013],[580,0.0115],[581,0.0115],[582,0.0082],[584,0.0082],[585,0.0049],[588,0.0049],[589,0.0016],[595,0.0016],[596,0.0026],[605,0.0026],[606,-0.0041],[611,-0.0041],[612,-0.0036],[613,-0.0036],[614,-0.006],[617,-0.006],[618,-0.0085],[619,-0.0085],[620,-0.0109],[622,-0.0109],[624,-0.0157],[625,-0.0157],[627,-0.0206],[628,-0.0216],[629,-0.0231],[630,-0.0262],[633,-0.0308],[634,-0.0339],[676,-0.0419],[681,-0.0461],[688,-0.0461],[689,-0.0504],[691,-0.0504],[692,-0.0546],[693,-0.0556],[694,-0.0575],[695,-0.0614],[697,-0.0652],[698,-0.069],[704,-0.0806],[705,-0.0806],[706,-0.0825],[707,-0.0825],[708,-0.0844],[709,-0.084],[712,-0.085],[713,-0.085],[714,-0.0853],[716,-0.0853],[717,-0.0856],[723,-0.0856],[725,-0.0851],[730,-0.0851],[731,-0.0845],[733,-0.0845],[734,-0.0839],[735,-0.0839],[736,-0.0833],[738,-0.0833],[739,-0.0827],[740,-0.0827],[741,-0.083],[742,-0.0813],[747,-0.0813],[748,-0.0795],[753,-0.0795],[754,-0.0778],[755,-0.0778],[757,-0.0773],[759,-0.0773],[770,-0.0041],[771,-0.0041],[772,-0.0036],[773,-0.0036],[774,-0.006],[777,-0.006],[778,-0.0085],[779,-0.0085],[780,-0.0109],[782,-0.0109],[784,-0.0157],[785,-0.0157],[787,-0.0206],[788,-0.0216],[789,-0.0231],[790,-0.0262],[793,-0.0308],[794,-0.0339],[798,-0.0401],[799,-0.0401],[801,-0.0432],[802,-0.0432],[803,-0.0447],[830,-0.0422],[837,-0.0419],[887,-0.0354],[890,-0.0401],[891,-0.0401],[893,-0.0432],[894,-0.0432],[895,-0.0447],[922,-0.0422],[929,-0.0419]]
       }
       spine_01_ctrl.Location.Y {
        type=MovieSceneScriptingFloatChannel
        default=0
        thinned_from=741
        keys [[-2,-0.1505],[3,-0.1544],[9,-0.1545],[10,-0.1584],[12,-0.1585],[13,-0.1624],[14,-0.1625],[15,-0.1637],[21,-0.177],[24,-0.1804],[28,-0.1805],[29,-0.1795],[30,-0.1796],[31,-0.1791],[32,-0.1791],[33,-0.1772],[34,-0.1772],[35,-0.1754],[36,-0.1754],[37,-0.1735],[38,-0.1735],[39,-0.1716],[40,-0.1716],[41,-0.1697],[42,-0.1697],[43,-0.1678],[45,-0.1678],[46,-0.1659],[47,-0.1659],[48,-0.1651],[51,-0.1641],[56,-0.1636],[59,-0.1642],[60,-0.1642],[63,-0.1649],[64,-0.1649],[65,-0.1668],[69,-0.1667],[70,-0.1688],[74,-0.1688],[75,-0.1707],[78,-0.1707],[82,-0.1703],[83,-0.1721],[84,-0.1721],[85,-0.1737],[86,-0.1737],[89,-0.1788],[90,-0.1788],[92,-0.1821],[96,-0.1817],[100,-0.1792],[103,-0.179],[105,-0.1798],[107,-0.1818],[108,-0.1833],[109,-0.1843],[111,-0.1874],[112,-0.1874],[113,-0.19],[114,-0.19],[115,-0.1926],[119,-0.1929],[120,-0.1904],[121,-0.1906],[124,-0.183],[125,-0.1778],[127,-0.1726],[133,-0.1534],[134,-0.1534],[135,-0.1502],[138,-0.1502],[139,-0.1534],[141,-0.1534],[142,-0.1566],[143,-0.1566],[144,-0.16],[145,-0.16],[146,-0.1634],[147,-0.1634],[148,-0.1667],[149,-0.1667],[150,-0.1701],[151,-0.1701],[152,-0.1735],[153,-0.1735],[154,-0.1769],[156,-0.1769],[157,-0.1803],[159,-0.1806],[160,-0.1829],[163,-0.1829],[164,-0.1853],[167,-0.1853],[168,-0.1876],[174,-0.1877],[175,-0.1868],[176,-0.1868],[177,-0.1844],[179,-0.1844],[180,-0.182],[181,-0.1821],[182,-0.1796],[183,-0.1797],[184,-0.1772],[185,-0.1773],[186,-0.1748],[187,-0.1749],[188,-0.1725],[189,-0.1725],[190,-0.1701],[191,-0.1698],[193,-0.1669],[194,-0.167],[195,-0.1655],[196,-0.1656],[197,-0.1641],[198,-0.1642],[199,-0.1627],[201,-0.1627],[202,-0.1613],[205,-0.1613],[206,-0.1599],[214,-0.1599],[215,-0.1667],[218,-0.1667],[219,-0.1735],[221,-0.1735],[222,-0.1803],[230,-0.2175],[237,-0.2172],[238,-0.2124],[239,-0.2136],[241,-0.2106],[242,-0.2104],[244,-0.2073],[245,-0.2073],[246,-0.2057],[247,-0.2057],[248,-0.2043],[251,-0.204],[252,-0.2025],[254,-0.2024],[255,-0.2036],[261,-0.2034],[262,-0.201],[266,-0.2009],[267,-0.1985],[269,-0.1985],[270,-0.1961],[276,-0.1961],[277,-0.187],[286,-0.187],[289,-0.1866],[290,-0.1878],[293,-0.1878],[294,-0.189],[295,-0.1889],[296,-0.1901],[297,-0.1901],[299,-0.1924],[300,-0.1924],[303,-0.1957],[305,-0.1969],[306,-0.1982],[309,-0.2],[323,-0.1505],[328,-0.1544],[334,-0.1545],[335,-0.1584],[337,-0.1585],[338,-0.1624],[339,-0.1625],[340,-0.1637],[346,-0.177],[349,-0.1804],[353,-0.1805],[354,-0.1795],[355,-0.1796],[356,-0.1791],[357,-0.1791],[358,-0.1772],[359,-0.1772],[360,-0.1754],[361,-0.1754],[362,-0.1735],[363,-0.1735],[364,-0.1716],[365,-0.1716],[366,-0.1697],[367,-0.1697],[368,-0.1678],[370,-0.1678],[371,-0.1659],[372,-0.1659],[373,-0.1651],[376,-0.1641],[381,-0.1636],[384,-0.1642],[385,-0.1642],[388,-0.1649],[389,-0.1649],[390,-0.1668],[394,-0.1667],[395,-0.1688],[399,-0.1688],[400,-0.1707],[403,-0.1707],[407,-0.1703],[408,-0.1721],[409,-0.1721],[410,-0.1737],[411,-0.1737],[414,-0.1788],[415,-0.1788],[417,-0.1821],[421,-0.1817],[425,-0.1792],[428,-0.179],[430,-0.1798],[432,-0.1818],[433,-0.1833],[434,-0.1843],[436,-0.1874],[437,-0.1874],[438,-0.19],[439,-0.19],[440,-0.1926],[444,-0.1929],[445,-0.1904],[446,-0.1906],[449,-0.183],[450,-0.1778],[452,-0.1726],[458,-0.1534],[459,-0.1534],[460,-0.1502],[463,-0.1502],[464,-0.1534],[466,-0.1534],[467,-0.1566],[468,-0.1566],[469,-0.16],[470,-0.16],[471,-0.1634],[472,-0.1634],[473,-0.1667],[474,-0.1667],[475,-0.1701],[476,-0.1701],[477,-0.1735],[478,-0.1735],[479,-0.1769],[481,-0.1769],[482,-0.1803],[484,-0.1806],[485,-0.1829],[488,-0.1829],[489,-0.1853],[492,-0.1853],[493,-0.1876],[499,-0.1877],[500,-0.1868],[501,-0.1868],[502,-0.1844],[504,-0.1844],[505,-0.182],[506,-0.1821],[507,-0.1796],[508,-0.1797],[509,-0.1772],[510,-0.1773],[511,-0.1748],[512,-0.1749],[513,-0.1725],[514,-0.1725],[515,-0.1701],[516,-0.1698],[518,-0.1669],[519,-0.167],[520,-0.1655],[521,-0.1656],[522,-0.1641],[523,-0.1642],[524,-0.1627],[526,-0.1627],[527,-0.1613],[530,-0.1613],[531,-0.1599],[539,-0.1599],[540,-0.1667],[543,-0.1667],[544,-0.1735],[546,-0.1735],[547,-0.1803],[555,-0.2175],[562,-0.2172],[563,-0.2124],[564,-0.2136],[566,-0.2106],[567,-0.2104],[569,-0.2073],[570,-0.2073],[571,-0.2057],[572,-0.2057],[573,-0.2043],[576,-0.204],[577,-0.2025],[579,-0.2024],[580,-0.2036],[586,-0.2034],[587,-0.201],[591,-0.2009],[592,-0.1985],[594,-0.1985],[595,-0.1961],[601,-0.1961],[602,-0.187],[611,-0.187],[614,-0.1866],[615,-0.1878],[618,-0.1878],[619,-0.189],[620,-0.1889],[621,-0.1901],[622,-0.1901],[624,-0.1924],[625,-0.1924],[628,-0.1957],[630,-0.1969],[631,-0.1982],[634,-0.2],[676,-0.1505],[681,-0.1544],[687,-0.1545],[688,-0.1584],[690,-0.1585],[691,-0.1624],[692,-0.1625],[693,-0.1637],[699,-0.177],[702,-0.1804],[706,-0.1805],[707,-0.1795],[708,-0.1796],[709,-0.1791],[710,-0.1791],[711,-0.1772],[712,-0.1772],[713,-0.1754],[714,-0.1754],[715,-0.1735],[716,-0.1735],[717,-0.1716],[718,-0.1716],[719,-0.1697],[720,-0.1697],[721,-0.1678],[723,-0.1678],[724,-0.1659],[725,-0.1659],[726,-0.1651],[729,-0.1641],[734,-0.1636],[737,-0.1642],[738,-0.1642],[741,-0.1649],[742,-0.1649],[743,-0.1668],[747,-0.1667],[748,-0.1688],[752,-0.1688],[753,-0.1707],[759,-0.1703],[770,-0.187],[774,-0.1866],[775,-0.1878],[778,-0.1878],[779,-0.189],[780,-0.1889],[781,-0.1901],[782,-0.1901],[784,-0.1924],[785,-0.1924],[788,-0.1957],[790,-0.1969],[791,-0.1982],[799,-0.2032],[800,-0.2032],[801,-0.2038],[802,-0.2038],[803,-0.2044],[830,-0.1564],[837,-0.1505],[887,-0.2007],[891,-0.2032],[892,-0.2032],[893,-0.2038],[894,-0.2038],[895,-0.2044],[922,-0.1564],[929,-0.1505]]
       }
       spine_01_ctrl.Location.Z {
        type=MovieSceneScriptingFloatChannel
        default=0
        thinned_from=741
        keys [[-2,0.0028],[8,0.0026],[9,0.005],[11,0.0049],[12,0.0072],[13,0.0071],[14,0.0094],[15,0.0098],[18,0.0148],[19,0.0181],[27,0.0315],[28,0.0315],[30,0.0349],[31,0.0349],[36,0.0373],[37,0.0373],[38,0.0378],[44,0.038],[45,0.0375],[46,0.0376],[47,0.038],[48,0.0365],[51,0.0366],[52,0.035],[54,0.0351],[55,0.0335],[56,0.0336],[57,0.032],[58,0.032],[60,0.029],[61,0.029],[62,0.0275],[64,0.0275],[65,0.0233],[68,0.0233],[69,0.0192],[74,0.0191],[75,0.0149],[78,0.0149],[79,0.0154],[80,0.0154],[81,0.0131],[84,0.013],[85,0.0107],[86,0.0107],[87,0.0084],[88,0.0083],[89,0.006],[90,0.006],[92,0.0014],[93,0.0014],[94,-0.0009],[95,-0.0022],[98,-0.0088],[99,-0.0132],[101,-0.0176],[102,-0.022],[107,-0.0331],[108,-0.0331],[109,-0.0354],[110,-0.0354],[111,-0.0359],[115,-0.0361],[116,-0.0318],[117,-0.0318],[118,-0.0275],[119,-0.0275],[121,-0.0189],[122,-0.0188],[124,-0.0101],[125,-0.0099],[126,-0.0055],[127,-0.0037],[133,0.0102],[134,0.0102],[135,0.0125],[139,0.0124],[140,0.0102],[142,0.0101],[143,0.0096],[144,0.0074],[145,0.0074],[146,0.0053],[147,0.0053],[148,0.0031],[149,0.0031],[150,0.0009],[151,0.0009],[152,-0.0013],[153,-0.0013],[154,-0.0035],[156,-0.0035],[157,-0.0057],[158,-0.0057],[159,-0.0063],[160,-0.0064],[161,-0.0074],[163,-0.0074],[164,-0.0085],[168,-0.0086],[169,-0.0075],[172,-0.0075],[173,-0.0065],[174,-0.0065],[175,-0.0059],[176,-0.0045],[177,-0.0044],[178,-0.0029],[180,-0.0028],[181,-0.0014],[182,-0.0013],[183,0.0001],[184,0.0002],[185,0.0017],[186,0.0017],[187,0.0032],[189,0.0033],[190,0.0048],[191,0.0048],[192,0.0053],[193,0.0053],[194,0.0058],[195,0.0058],[196,0.0063],[197,0.0063],[198,0.0067],[204,0.0068],[205,0.0073],[206,0.0073],[207,0.0078],[208,0.0064],[214,0.0064],[215,0.0049],[218,0.0049],[219,0.0047],[221,0.0047],[222,0.0031],[223,0.0034],[224,0.0001],[226,-0.0002],[227,-0.0035],[229,-0.0038],[230,-0.0071],[231,-0.0071],[232,-0.0102],[233,-0.0102],[234,-0.0134],[235,-0.0134],[236,-0.0165],[237,-0.0165],[238,-0.0195],[239,-0.0204],[240,-0.0247],[241,-0.0247],[242,-0.029],[243,-0.029],[244,-0.0333],[245,-0.0333],[246,-0.0376],[248,-0.0375],[249,-0.0419],[250,-0.0419],[251,-0.0462],[252,-0.0462],[253,-0.0505],[254,-0.0505],[255,-0.0515],[256,-0.0554],[258,-0.0554],[259,-0.0594],[262,-0.0593],[263,-0.0633],[270,-0.0631],[271,-0.0627],[276,-0.0627],[277,-0.058],[286,-0.058],[287,-0.0591],[291,-0.0591],[292,-0.0603],[294,-0.0603],[295,-0.0614],[296,-0.0615],[297,-0.0626],[298,-0.0627],[300,-0.065],[301,-0.065],[302,-0.0662],[305,-0.068],[306,-0.0692],[309,-0.071],[323,0.0028],[333,0.0026],[334,0.005],[336,0.0049],[337,0.0072],[338,0.0071],[339,0.0094],[340,0.0098],[343,0.0148],[344,0.0181],[352,0.0315],[353,0.0315],[355,0.0349],[356,0.0349],[361,0.0373],[362,0.0373],[363,0.0378],[369,0.038],[370,0.0375],[371,0.0376],[372,0.038],[373,0.0365],[376,0.0366],[377,0.035],[379,0.0351],[380,0.0335],[381,0.0336],[382,0.032],[383,0.032],[385,0.029],[386,0.029],[387,0.0275],[389,0.0275],[390,0.0233],[393,0.0233],[394,0.0192],[399,0.0191],[400,0.0149],[403,0.0149],[404,0.0154],[405,0.0154],[406,0.0131],[409,0.013],[410,0.0107],[411,0.0107],[412,0.0084],[413,0.0083],[414,0.006],[415,0.006],[417,0.0014],[418,0.0014],[419,-0.0009],[420,-0.0022],[423,-0.0088],[424,-0.0132],[426,-0.0176],[427,-0.022],[432,-0.0331],[433,-0.0331],[434,-0.0354],[435,-0.0354],[436,-0.0359],[440,-0.0361],[441,-0.0318],[442,-0.0318],[443,-0.0275],[444,-0.0275],[446,-0.0189],[447,-0.0188],[449,-0.0101],[450,-0.0099],[451,-0.0055],[452,-0.0037],[458,0.0102],[459,0.0102],[460,0.0125],[464,0.0124],[465,0.0102],[467,0.0101],[468,0.0096],[469,0.0074],[470,0.0074],[471,0.0053],[472,0.0053],[473,0.0031],[474,0.0031],[475,0.0009],[476,0.0009],[477,-0.0013],[478,-0.0013],[479,-0.0035],[481,-0.0035],[482,-0.0057],[483,-0.0057],[484,-0.0063],[485,-0.0064],[486,-0.0074],[488,-0.0074],[489,-0.0085],[493,-0.0086],[494,-0.0075],[497,-0.0075],[498,-0.0065],[499,-0.0065],[500,-0.0059],[501,-0.0045],[502,-0.0044],[503,-0.0029],[505,-0.0028],[506,-0.0014],[507,-0.0013],[508,0.0001],[509,0.0002],[510,0.0017],[511,0.0017],[512,0.0032],[514,0.0033],[515,0.0048],[516,0.0048],[517,0.0053],[518,0.0053],[519,0.0058],[520,0.0058],[521,0.0063],[522,0.0063],[523,0.0067],[529,0.0068],[530,0.0073],[531,0.0073],[532,0.0078],[533,0.0064],[539,0.0064],[540,0.0049],[543,0.0049],[544,0.0047],[546,0.0047],[547,0.0031],[548,0.0034],[549,0.0001],[551,-0.0002],[552,-0.0035],[554,-0.0038],[555,-0.0071],[556,-0.0071],[557,-0.0102],[558,-0.0102],[559,-0.0134],[560,-0.0134],[561,-0.0165],[562,-0.0165],[563,-0.0195],[564,-0.0204],[565,-0.0247],[566,-0.0247],[567,-0.029],[568,-0.029],[569,-0.0333],[570,-0.0333],[571,-0.0376],[573,-0.0375],[574,-0.0419],[575,-0.0419],[576,-0.0462],[577,-0.0462],[578,-0.0505],[579,-0.0505],[580,-0.0515],[581,-0.0554],[583,-0.0554],[584,-0.0594],[587,-0.0593],[588,-0.0633],[595,-0.0631],[596,-0.0627],[601,-0.0627],[602,-0.058],[611,-0.058],[612,-0.0591],[616,-0.0591],[617,-0.0603],[619,-0.0603],[620,-0.0614],[621,-0.0615],[622,-0.0626],[623,-0.0627],[625,-0.065],[626,-0.065],[627,-0.0662],[630,-0.068],[631,-0.0692],[634,-0.071],[676,0.0028],[686,0.0026],[687,0.005],[689,0.0049],[690,0.0072],[691,0.0071],[692,0.0094],[693,0.0098],[696,0.0148],[697,0.0181],[705,0.0315],[706,0.0315],[708,0.0349],[709,0.0349],[714,0.0373],[715,0.0373],[716,0.0378],[722,0.038],[723,0.0375],[724,0.0376],[725,0.038],[726,0.0365],[729,0.0366],[730,0.035],[732,0.0351],[733,0.0335],[734,0.0336],[735,0.032],[736,0.032],[738,0.029],[739,0.029],[740,0.0275],[742,0.0275],[743,0.0233],[746,0.0233],[747,0.0192],[752,0.0191],[753,0.0149],[755,0.0149],[757,0.0154],[759,0.0131],[770,-0.058],[771,-0.058],[772,-0.0591],[776,-0.0591],[777,-0.0603],[779,-0.0603],[780,-0.0614],[781,-0.0615],[782,-0.0626],[783,-0.0627],[785,-0.065],[786,-0.065],[787,-0.0662],[790,-0.068],[791,-0.0692],[798,-0.0735],[799,-0.0735],[801,-0.0747],[802,-0.0747],[803,-0.0753],[830,-0.0058],[837,0.0028],[887,-0.0717],[890,-0.0735],[891,-0.0735],[893,-0.0747],[894,-0.0747],[895,-0.0753],[922,-0.0058],[929,0.0028]]
       }
       spine_01_ctrl.Rotation.X {
        type=MovieSceneScriptingFloatChannel
        default=0
        thinned_from=712
        keys [[-2,3.3148],[5,3.4613],[7,3.442],[15,3.1695],[16,3.0825],[17,2.9892],[22,2.4642],[23,2.3646],[24,2.2721],[25,2.1856],[26,2.1038],[27,2.0265],[28,1.9571],[29,1.8927],[30,1.8362],[31,1.7865],[32,1.7457],[33,1.7111],[34,1.6814],[35,1.6564],[36,1.6354],[37,1.618],[39,1.5957],[40,1.5892],[42,1.5862],[45,1.5985],[46,1.6059],[47,1.6168],[49,1.6533],[50,1.6806],[51,1.7126],[52,1.7491],[54,1.8373],[55,1.8875],[56,1.9422],[58,2.0624],[62,2.3301],[67,2.6762],[70,2.8628],[72,2.9703],[74,3.0603],[77,3.1696],[81,3.2805],[82,3.3126],[83,3.3582],[84,3.4145],[85,3.4893],[86,3.5917],[87,3.7167],[88,3.8604],[89,4.017],[91,4.354],[94,4.8737],[97,5.3421],[99,5.6209],[100,5.744],[101,5.8538],[102,5.9461],[103,6.0183],[104,6.0665],[105,6.0886],[106,6.0829],[107,6.0475],[108,5.9895],[109,5.9177],[110,5.8329],[111,5.7365],[112,5.6282],[113,5.5079],[114,5.361],[115,5.1795],[116,4.968],[117,4.7304],[118,4.4752],[120,3.9238],[123,3.0627],[124,2.7801],[125,2.5047],[126,2.2403],[127,1.9936],[128,1.7641],[129,1.5538],[130,1.3638],[131,1.1932],[132,1.0431],[133,0.9158],[134,0.8124],[135,0.7341],[136,0.682],[137,0.6574],[138,0.6539],[139,0.6629],[140,0.6834],[141,0.7134],[142,0.754],[143,0.8039],[144,0.862],[145,0.9268],[147,1.0749],[149,1.2352],[153,1.5587],[156,1.7762],[157,1.8444],[158,1.9073],[160,2.0237],[161,2.0742],[162,2.1178],[164,2.192],[166,2.254],[167,2.2803],[168,2.2992],[169,2.3127],[171,2.3255],[172,2.3267],[177,2.285],[181,2.227],[183,2.1872],[187,2.0984],[189,2.0668],[190,2.0542],[191,2.0459],[193,2.0393],[197,2.048],[207,2.1213],[209,2.1464],[213,2.2182],[214,2.2305],[215,2.2356],[216,2.2317],[217,2.2141],[218,2.1798],[219,2.1288],[220,2.0628],[223,1.8908],[224,1.8363],[226,1.7393],[227,1.7058],[228,1.6868],[229,1.684],[230,1.701],[231,1.7386],[232,1.7972],[233,1.8753],[234,1.9672],[235,2.0708],[236,2.1843],[237,2.3095],[239,2.5774],[241,2.8636],[245,3.4655],[246,3.6207],[247,3.7685],[249,4.0358],[250,4.1557],[251,4.2655],[252,4.3651],[254,4.538],[256,4.6695],[258,4.7728],[260,4.8438],[261,4.8657],[262,4.8775],[263,4.8795],[265,4.8517],[266,4.8238],[268,4.7409],[271,4.5759],[273,4.4405],[275,4.283],[277,4.1459],[278,4.0878],[280,3.9963],[282,3.9334],[284,3.8987],[289,3.8787],[292,3.8948],[295,3.9377],[299,4.0304],[309,4.3635],[315,4.341],[327,3.9011],[340,3.1695],[341,3.0825],[342,2.9892],[347,2.4642],[348,2.3646],[349,2.2721],[350,2.1856],[351,2.1038],[352,2.0265],[353,1.9571],[354,1.8927],[355,1.8362],[356,1.7865],[357,1.7457],[358,1.7111],[359,1.6814],[360,1.6564],[361,1.6354],[362,1.618],[364,1.5957],[365,1.5892],[367,1.5862],[370,1.5985],[371,1.6059],[372,1.6168],[374,1.6533],[375,1.6806],[376,1.7126],[377,1.7491],[379,1.8373],[380,1.8875],[381,1.9422],[383,2.0624],[387,2.3301],[392,2.6762],[395,2.8628],[397,2.9703],[399,3.0603],[402,3.1696],[406,3.2805],[407,3.3126],[408,3.3582],[409,3.4145],[410,3.4893],[411,3.5917],[412,3.7167],[413,3.8604],[414,4.017],[416,4.354],[419,4.8737],[422,5.3421],[424,5.6209],[425,5.744],[426,5.8538],[427,5.9461],[428,6.0183],[429,6.0665],[430,6.0886],[431,6.0829],[432,6.0475],[433,5.9895],[434,5.9177],[435,5.8329],[436,5.7365],[437,5.6282],[438,5.5079],[439,5.361],[440,5.1795],[441,4.968],[442,4.7304],[443,4.4752],[445,3.9238],[448,3.0627],[449,2.7801],[450,2.5047],[451,2.2403],[452,1.9936],[453,1.7641],[454,1.5538],[455,1.3638],[456,1.1932],[457,1.0431],[458,0.9158],[459,0.8124],[460,0.7341],[461,0.682],[462,0.6574],[463,0.6539],[464,0.6629],[465,0.6834],[466,0.7134],[467,0.754],[468,0.8039],[469,0.862],[470,0.9268],[472,1.0749],[474,1.2352],[478,1.5587],[481,1.7762],[482,1.8444],[483,1.9073],[485,2.0237],[486,2.0742],[487,2.1178],[489,2.192],[491,2.254],[492,2.2803],[493,2.2992],[494,2.3127],[496,2.3255],[497,2.3267],[502,2.285],[506,2.227],[508,2.1872],[512,2.0984],[514,2.0668],[515,2.0542],[516,2.0459],[518,2.0393],[522,2.048],[532,2.1213],[534,2.1464],[538,2.2182],[539,2.2305],[540,2.2356],[541,2.2317],[542,2.2141],[543,2.1798],[544,2.1288],[545,2.0628],[548,1.8908],[549,1.8363],[551,1.7393],[552,1.7058],[553,1.6868],[554,1.684],[555,1.701],[556,1.7386],[557,1.7972],[558,1.8753],[559,1.9672],[560,2.0708],[561,2.1843],[562,2.3095],[564,2.5774],[566,2.8636],[570,3.4655],[571,3.6207],[572,3.7685],[574,4.0358],[575,4.1557],[576,4.2655],[577,4.3651],[579,4.538],[581,4.6695],[583,4.7728],[585,4.8438],[586,4.8657],[587,4.8775],[588,4.8795],[590,4.8517],[591,4.8238],[593,4.7409],[596,4.5759],[598,4.4405],[600,4.283],[602,4.1459],[603,4.0878],[605,3.9963],[607,3.9334],[609,3.8987],[611,3.8928],[612,3.9022],[613,3.9252],[614,3.9626],[615,4.0148],[616,4.0808],[617,4.1602],[618,4.2519],[619,4.3555],[620,4.4706],[621,4.5969],[622,4.7334],[624,5.035],[626,5.368],[630,6.0985],[634,6.873],[676,-5.29],[683,-5.1435],[685,-5.1629],[693,-5.4353],[695,-5.6157],[701,-6.2403],[704,-6.501],[707,-6.7121],[709,-6.8184],[712,-6.9235],[716,-6.9995],[723,-7.0064],[727,-6.9516],[730,-6.8558],[734,-6.6626],[748,-5.7421],[751,-5.587],[755,-5.4352],[759,-5.3243],[770,-4.7128],[772,-4.7026],[773,-4.6797],[774,-4.6423],[775,-4.59],[776,-4.524],[777,-4.4446],[778,-4.3529],[779,-4.2493],[780,-4.1342],[781,-4.0079],[782,-3.8715],[783,-3.7253],[784,-3.5698],[785,-3.4069],[787,-3.0609],[789,-2.6943],[792,-2.1222],[796,-1.3431],[798,-0.9622],[799,-0.7745],[801,-0.4086],[802,-0.2286],[803,-0.0526],[829,-4.2183],[830,-4.4504],[837,-5.29],[887,7.0677],[894,8.3762],[895,8.5522],[921,4.3865],[922,4.1544],[929,3.3148]]
       }
       spine_01_ctrl.Rotation.Y {
        type=MovieSceneScriptingFloatChannel
        default=0
        thinned_from=666
        keys [[-2,0.6232],[7,1.9036],[18,4.675],[19,4.9657],[20,5.2354],[21,5.4773],[22,5.6865],[23,5.8562],[24,5.9794],[25,6.0707],[26,6.1463],[27,6.2068],[29,6.2786],[31,6.2922],[33,6.2612],[36,6.1669],[55,5.2223],[70,4.3726],[74,4.1778],[82,3.8577],[83,3.8311],[84,3.8152],[85,3.8116],[87,3.8263],[88,3.8407],[89,3.8659],[92,3.9704],[93,3.987],[94,3.9819],[95,3.9467],[96,3.8711],[97,3.7468],[98,3.5818],[99,3.3837],[100,3.1536],[101,2.8952],[102,2.6161],[103,2.3146],[104,1.9998],[105,1.6771],[106,1.363],[107,1.0586],[108,0.7695],[109,0.5026],[110,0.2591],[111,0.042],[112,-0.1453],[113,-0.3106],[114,-0.456],[115,-0.5726],[116,-0.6581],[117,-0.7079],[118,-0.7208],[119,-0.6936],[120,-0.6247],[121,-0.522],[122,-0.393],[123,-0.2382],[124,-0.0568],[125,0.1461],[126,0.3653],[127,0.5937],[128,0.825],[129,1.0523],[130,1.273],[131,1.484],[132,1.6839],[133,1.8744],[134,2.0565],[135,2.2308],[136,2.3978],[137,2.5576],[138,2.7101],[139,2.8564],[140,2.9953],[141,3.1272],[142,3.2518],[143,3.3685],[144,3.4775],[145,3.5776],[146,3.6702],[147,3.755],[157,4.4761],[162,4.7879],[172,5.3372],[175,5.5308],[182,6.0761],[189,6.4743],[195,6.7394],[201,6.9217],[206,6.9936],[210,6.9876],[212,6.9596],[214,6.8912],[216,6.7744],[217,6.6957],[218,6.5924],[219,6.4584],[220,6.2877],[221,6.0712],[222,5.8162],[223,5.5316],[227,4.2947],[228,4.0085],[229,3.7505],[230,3.5319],[231,3.3624],[232,3.232],[233,3.1247],[235,2.9482],[237,2.7864],[239,2.6387],[244,2.3143],[246,2.1699],[249,1.9423],[250,1.8766],[251,1.8201],[252,1.7708],[253,1.729],[254,1.6929],[255,1.6653],[256,1.6516],[257,1.6477],[258,1.6577],[260,1.6856],[261,1.6905],[262,1.6754],[263,1.6408],[264,1.5829],[265,1.5038],[266,1.4077],[267,1.289],[268,1.1635],[269,1.0416],[270,0.9236],[271,0.8135],[272,0.7124],[273,0.618],[274,0.5305],[275,0.4521],[276,0.375],[277,0.3075],[278,0.2459],[279,0.1944],[280,0.1512],[281,0.1168],[282,0.0895],[283,0.0714],[284,0.0629],[285,0.0643],[286,0.0739],[287,0.0955],[288,0.1207],[290,0.1697],[291,0.1948],[292,0.2213],[294,0.2769],[295,0.3054],[296,0.3366],[297,0.3733],[298,0.4146],[299,0.4596],[300,0.5068],[301,0.5565],[307,0.866],[309,0.974],[323,0.6232],[332,1.9036],[343,4.675],[344,4.9657],[345,5.2354],[346,5.4773],[347,5.6865],[348,5.8562],[349,5.9794],[350,6.0707],[351,6.1463],[352,6.2068],[354,6.2786],[356,6.2922],[358,6.2612],[361,6.1669],[380,5.2223],[395,4.3726],[399,4.1778],[407,3.8577],[408,3.8311],[409,3.8152],[410,3.8116],[412,3.8263],[413,3.8407],[414,3.8659],[417,3.9704],[418,3.987],[419,3.9819],[420,3.9467],[421,3.8711],[422,3.7468],[423,3.5818],[424,3.3837],[425,3.1536],[426,2.8952],[427,2.6161],[428,2.3146],[429,1.9998],[430,1.6771],[431,1.363],[432,1.0586],[433,0.7695],[434,0.5026],[435,0.2591],[436,0.042],[437,-0.1453],[438,-0.3106],[439,-0.456],[440,-0.5726],[441,-0.6581],[442,-0.7079],[443,-0.7208],[444,-0.6936],[445,-0.6247],[446,-0.522],[447,-0.393],[448,-0.2382],[449,-0.0568],[450,0.1461],[451,0.3653],[452,0.5937],[453,0.825],[454,1.0523],[455,1.273],[456,1.484],[457,1.6839],[458,1.8744],[459,2.0565],[460,2.2308],[461,2.3978],[462,2.5576],[463,2.7101],[464,2.8564],[465,2.9953],[466,3.1272],[467,3.2518],[468,3.3685],[469,3.4775],[470,3.5776],[471,3.6702],[472,3.755],[482,4.4761],[487,4.7879],[497,5.3372],[500,5.5308],[507,6.0761],[514,6.4743],[520,6.7394],[526,6.9217],[531,6.9936],[535,6.9876],[537,6.9596],[539,6.8912],[541,6.7744],[542,6.6957],[543,6.5924],[544,6.4584],[545,6.2877],[546,6.0712],[547,5.8162],[548,5.5316],[552,4.2947],[553,4.0085],[554,3.7505],[555,3.5319],[556,3.3624],[557,3.232],[558,3.1247],[560,2.9482],[562,2.7864],[564,2.6387],[569,2.3143],[571,2.1699],[574,1.9423],[575,1.8766],[576,1.8201],[577,1.7708],[578,1.729],[579,1.6929],[580,1.6653],[581,1.6516],[582,1.6477],[583,1.6577],[585,1.6856],[586,1.6905],[587,1.6754],[588,1.6408],[589,1.5829],[590,1.5038],[591,1.4077],[592,1.289],[593,1.1635],[594,1.0416],[595,0.9236],[596,0.8135],[597,0.7124],[598,0.618],[599,0.5305],[600,0.4521],[601,0.375],[602,0.3075],[603,0.2459],[604,0.1944],[605,0.1512],[606,0.1168],[607,0.0895],[608,0.0714],[609,0.0629],[610,0.0643],[611,0.0739],[612,0.0955],[613,0.1207],[615,0.1697],[616,0.1948],[617,0.2213],[619,0.2769],[620,0.3054],[621,0.3366],[622,0.3733],[623,0.4146],[624,0.4596],[625,0.5068],[626,0.5565],[632,0.866],[634,0.974],[676,0.6232],[685,1.9036],[696,4.675],[697,4.9657],[698,5.2354],[699,5.4773],[700,5.6865],[701,5.8562],[702,5.9794],[703,6.0707],[704,6.1463],[705,6.2068],[707,6.2786],[709,6.2922],[711,6.2612],[714,6.1669],[733,5.2223],[748,4.3726],[752,4.1778],[759,3.894],[803,1.4487],[830,0.7138],[837,0.6232],[887,1.0295],[892,1.3012],[894,1.4021],[895,1.4487],[922,0.7138],[929,0.6232]]
       }
       spine_01_ctrl.Rotation.Z {
        type=MovieSceneScriptingFloatChannel
        default=0
        thinned_from=666
        keys [[-2,0.4973],[3,0.3926],[4,0.3484],[5,0.2982],[6,0.2431],[7,0.1837],[9,0.0581],[10,-0.0033],[11,-0.0653],[12,-0.1267],[14,-0.2519],[15,-0.315],[17,-0.4433],[19,-0.5734],[20,-0.6375],[21,-0.6997],[23,-0.8163],[24,-0.8685],[25,-0.9159],[26,-0.9568],[27,-0.9904],[28,-1.0155],[29,-1.0345],[30,-1.0467],[31,-1.0548],[41,-1.0832],[70,-1.1006],[74,-1.0974],[83,-1.1299],[84,-1.1833],[85,-1.265],[86,-1.3728],[87,-1.5003],[88,-1.6514],[89,-1.8267],[90,-2.0165],[92,-2.4087],[93,-2.5958],[94,-2.7659],[95,-2.913],[96,-3.0284],[97,-3.1065],[98,-3.1531],[99,-3.1719],[100,-3.1638],[101,-3.1289],[102,-3.0672],[103,-2.9785],[104,-2.8626],[105,-2.7272],[106,-2.5792],[107,-2.4217],[111,-1.7579],[114,-1.2818],[115,-1.1275],[116,-0.9799],[117,-0.8457],[118,-0.7313],[119,-0.6398],[120,-0.5747],[121,-0.5319],[122,-0.5042],[123,-0.4952],[124,-0.5057],[125,-0.5321],[126,-0.5701],[127,-0.6156],[128,-0.6632],[129,-0.7092],[130,-0.7491],[131,-0.7828],[132,-0.8084],[133,-0.8281],[134,-0.8393],[135,-0.8449],[136,-0.8428],[137,-0.8353],[138,-0.8237],[152,-0.6077],[154,-0.5978],[155,-0.5951],[156,-0.5943],[157,-0.5948],[158,-0.5965],[159,-0.5999],[160,-0.6052],[161,-0.6123],[163,-0.6312],[165,-0.6548],[166,-0.6684],[168,-0.6993],[169,-0.7177],[170,-0.7376],[171,-0.7608],[172,-0.7896],[173,-0.8233],[174,-0.8603],[175,-0.9049],[176,-0.9585],[177,-1.0182],[179,-1.1437],[180,-1.2023],[181,-1.2584],[182,-1.3109],[183,-1.3559],[184,-1.393],[187,-1.486],[189,-1.5404],[191,-1.5875],[193,-1.6249],[195,-1.6521],[197,-1.6701],[199,-1.6772],[201,-1.6718],[203,-1.6535],[204,-1.6394],[205,-1.6221],[206,-1.6],[207,-1.5735],[208,-1.5438],[209,-1.5101],[210,-1.4704],[211,-1.4266],[212,-1.3751],[213,-1.3127],[214,-1.2362],[215,-1.1457],[216,-1.0445],[217,-0.9325],[218,-0.8007],[219,-0.6488],[220,-0.4661],[221,-0.2585],[222,-0.0295],[223,0.2115],[224,0.4663],[225,0.7268],[226,0.9912],[227,1.2517],[228,1.5051],[229,1.7468],[230,1.969],[231,2.1673],[232,2.3404],[233,2.4782],[234,2.593],[235,2.6896],[236,2.7681],[237,2.8331],[238,2.8819],[239,2.916],[240,2.9371],[241,2.9449],[242,2.9412],[243,2.927],[244,2.9037],[245,2.8715],[246,2.8309],[247,2.783],[248,2.7285],[249,2.6681],[251,2.5318],[257,2.0704],[258,1.9986],[259,1.9311],[260,1.8684],[261,1.8119],[262,1.7631],[263,1.7212],[264,1.6886],[265,1.6653],[266,1.6525],[267,1.6504],[284,1.6806],[290,1.7032],[292,1.6992],[293,1.692],[295,1.6682],[296,1.6491],[297,1.6223],[298,1.5882],[299,1.5482],[300,1.5027],[301,1.4521],[302,1.3974],[304,1.2777],[307,1.0781],[309,0.941],[323,0.4973],[328,0.3926],[329,0.3484],[330,0.2982],[331,0.2431],[332,0.1837],[334,0.0581],[335,-0.0033],[336,-0.0653],[337,-0.1267],[339,-0.2519],[340,-0.315],[342,-0.4433],[344,-0.5734],[345,-0.6375],[346,-0.6997],[348,-0.8163],[349,-0.8685],[350,-0.9159],[351,-0.9568],[352,-0.9904],[353,-1.0155],[354,-1.0345],[355,-1.0467],[356,-1.0548],[366,-1.0832],[395,-1.1006],[399,-1.0974],[408,-1.1299],[409,-1.1833],[410,-1.265],[411,-1.3728],[412,-1.5003],[413,-1.6514],[414,-1.8267],[415,-2.0165],[417,-2.4087],[418,-2.5958],[419,-2.7659],[420,-2.913],[421,-3.0284],[422,-3.1065],[423,-3.1531],[424,-3.1719],[425,-3.1638],[426,-3.1289],[427,-3.0672],[428,-2.9785],[429,-2.8626],[430,-2.7272],[431,-2.5792],[432,-2.4217],[436,-1.7579],[439,-1.2818],[440,-1.1275],[441,-0.9799],[442,-0.8457],[443,-0.7313],[444,-0.6398],[445,-0.5747],[446,-0.5319],[447,-0.5042],[448,-0.4952],[449,-0.5057],[450,-0.5321],[451,-0.5701],[452,-0.6156],[453,-0.6632],[454,-0.7092],[455,-0.7491],[456,-0.7828],[457,-0.8084],[458,-0.8281],[459,-0.8393],[460,-0.8449],[461,-0.8428],[462,-0.8353],[463,-0.8237],[477,-0.6077],[479,-0.5978],[480,-0.5951],[481,-0.5943],[482,-0.5948],[483,-0.5965],[484,-0.5999],[485,-0.6052],[486,-0.6123],[488,-0.6312],[490,-0.6548],[491,-0.6684],[493,-0.6993],[494,-0.7177],[495,-0.7376],[496,-0.7608],[497,-0.7896],[498,-0.8233],[499,-0.8603],[500,-0.9049],[501,-0.9585],[502,-1.0182],[504,-1.1437],[505,-1.2023],[506,-1.2584],[507,-1.3109],[508,-1.3559],[509,-1.393],[512,-1.486],[514,-1.5404],[516,-1.5875],[518,-1.6249],[520,-1.6521],[522,-1.6701],[524,-1.6772],[526,-1.6718],[528,-1.6535],[529,-1.6394],[530,-1.6221],[531,-1.6],[532,-1.5735],[533,-1.5438],[534,-1.5101],[535,-1.4704],[536,-1.4266],[537,-1.3751],[538,-1.3127],[539,-1.2362],[540,-1.1457],[541,-1.0445],[542,-0.9325],[543,-0.8007],[544,-0.6488],[545,-0.4661],[546,-0.2585],[547,-0.0295],[548,0.2115],[549,0.4663],[550,0.7268],[551,0.9912],[552,1.2517],[553,1.5051],[554,1.7468],[555,1.969],[556,2.1673],[557,2.3404],[558,2.4782],[559,2.593],[560,2.6896],[561,2.7681],[562,2.8331],[563,2.8819],[564,2.916],[565,2.9371],[566,2.9449],[567,2.9412],[568,2.927],[569,2.9037],[570,2.8715],[571,2.8309],[572,2.783],[573,2.7285],[574,2.6681],[576,2.5318],[582,2.0704],[583,1.9986],[584,1.9311],[585,1.8684],[586,1.8119],[587,1.7631],[588,1.7212],[589,1.6886],[590,1.6653],[591,1.6525],[592,1.6504],[609,1.6806],[615,1.7032],[617,1.6992],[618,1.692],[620,1.6682],[621,1.6491],[622,1.6223],[623,1.5882],[624,1.5482],[625,1.5027],[626,1.4521],[627,1.3974],[629,1.2777],[632,1.0781],[634,0.941],[676,9.9541],[786,9.6908],[830,9.9752],[837,9.9858],[887,0.8721],[889,0.7394],[890,0.6756],[891,0.6139],[892,0.5551],[893,0.5004],[894,0.4486],[895,0.4007],[922,0.4867],[929,0.4973]]
       }
       spine_01_ctrl.Scale.X {type=MovieSceneScriptingFloatChannel, default=1, const [1,741]}
       spine_01_ctrl.Scale.Y {type=MovieSceneScriptingFloatChannel, default=1, const [1,741]}
       spine_01_ctrl.Scale.Z {type=MovieSceneScriptingFloatChannel, default=1, const [1,741]}
       spine_02_ctrl.Location.X {
        type=MovieSceneScriptingFloatChannel
        default=0
        thinned_from=739
        keys [[-2,-0.0821],[3,-0.0876],[8,-0.0896],[9,-0.0905],[10,-0.0919],[11,-0.0938],[12,-0.0961],[13,-0.0988],[15,-0.1048],[16,-0.1116],[17,-0.1116],[18,-0.1184],[19,-0.1184],[21,-0.132],[22,-0.132],[23,-0.1387],[24,-0.1387],[25,-0.1455],[27,-0.1455],[28,-0.1523],[30,-0.1523],[31,-0.1516],[32,-0.1531],[34,-0.1531],[35,-0.1546],[46,-0.1546],[47,-0.1561],[48,-0.1536],[56,-0.1536],[57,-0.151],[60,-0.151],[61,-0.1484],[62,-0.1484],[63,-0.1491],[64,-0.1491],[65,-0.1459],[69,-0.1459],[70,-0.1427],[74,-0.1427],[75,-0.1395],[78,-0.1395],[80,-0.1397],[81,-0.1375],[83,-0.1375],[84,-0.1397],[85,-0.1397],[87,-0.1441],[88,-0.1484],[89,-0.1506],[92,-0.1637],[94,-0.1681],[95,-0.1728],[99,-0.1728],[100,-0.1627],[101,-0.1627],[102,-0.1527],[103,-0.1527],[104,-0.1426],[105,-0.1426],[107,-0.1224],[108,-0.1224],[110,-0.1023],[111,-0.0959],[113,-0.0764],[114,-0.0764],[116,-0.057],[117,-0.057],[118,-0.0473],[119,-0.0473],[120,-0.0376],[122,-0.0376],[123,-0.0279],[129,-0.0279],[130,-0.0285],[131,-0.0285],[132,-0.0292],[133,-0.0292],[134,-0.0298],[138,-0.0298],[139,-0.0305],[141,-0.0305],[142,-0.0311],[144,-0.0311],[145,-0.0314],[146,-0.0314],[147,-0.0317],[148,-0.0317],[149,-0.032],[150,-0.032],[152,-0.0326],[153,-0.0326],[154,-0.0329],[155,-0.0329],[158,-0.0338],[159,-0.0337],[160,-0.035],[165,-0.035],[166,-0.0362],[174,-0.0362],[175,-0.0369],[176,-0.0369],[177,-0.0414],[180,-0.0414],[181,-0.0459],[187,-0.0459],[188,-0.0503],[190,-0.0503],[191,-0.0492],[192,-0.05],[194,-0.05],[195,-0.0508],[198,-0.0508],[199,-0.05],[201,-0.05],[203,-0.0483],[204,-0.0483],[206,-0.0467],[207,-0.0446],[208,-0.0446],[209,-0.0426],[210,-0.0426],[215,-0.033],[222,-0.033],[223,-0.032],[226,-0.032],[227,-0.0304],[229,-0.0304],[230,-0.0287],[231,-0.0287],[232,-0.0271],[234,-0.0271],[235,-0.0254],[236,-0.0254],[238,-0.0221],[239,-0.0217],[240,-0.0204],[242,-0.0187],[245,-0.0147],[246,-0.013],[248,-0.0104],[250,-0.0087],[251,-0.0087],[253,-0.0095],[254,-0.0108],[255,-0.0125],[257,-0.0171],[258,-0.0171],[260,-0.0217],[261,-0.0217],[262,-0.024],[263,-0.024],[264,-0.0263],[268,-0.0263],[269,-0.0285],[270,-0.0285],[271,-0.0273],[274,-0.0273],[275,-0.0303],[281,-0.0303],[282,-0.0332],[285,-0.0332],[286,-0.0362],[287,-0.0356],[288,-0.0356],[289,-0.0392],[292,-0.0392],[293,-0.0427],[294,-0.0427],[295,-0.0463],[296,-0.0463],[297,-0.0499],[298,-0.0499],[299,-0.0535],[300,-0.0535],[302,-0.0606],[303,-0.0612],[304,-0.0653],[306,-0.0693],[307,-0.0733],[309,-0.0773],[323,-0.0821],[328,-0.0876],[333,-0.0896],[334,-0.0905],[335,-0.0919],[336,-0.0938],[337,-0.0961],[338,-0.0988],[340,-0.1048],[341,-0.1116],[342,-0.1116],[343,-0.1184],[344,-0.1184],[346,-0.132],[347,-0.132],[348,-0.1387],[349,-0.1387],[350,-0.1455],[352,-0.1455],[353,-0.1523],[355,-0.1523],[356,-0.1516],[357,-0.1531],[359,-0.1531],[360,-0.1546],[371,-0.1546],[372,-0.1561],[373,-0.1536],[381,-0.1536],[382,-0.151],[385,-0.151],[386,-0.1484],[387,-0.1484],[388,-0.1491],[389,-0.1491],[390,-0.1459],[394,-0.1459],[395,-0.1427],[399,-0.1427],[400,-0.1395],[403,-0.1395],[405,-0.1397],[406,-0.1375],[408,-0.1375],[409,-0.1397],[410,-0.1397],[412,-0.1441],[413,-0.1484],[414,-0.1506],[417,-0.1637],[419,-0.1681],[420,-0.1728],[424,-0.1728],[425,-0.1627],[426,-0.1627],[427,-0.1527],[428,-0.1527],[429,-0.1426],[430,-0.1426],[432,-0.1224],[433,-0.1224],[435,-0.1023],[436,-0.0959],[438,-0.0764],[439,-0.0764],[441,-0.057],[442,-0.057],[443,-0.0473],[444,-0.0473],[445,-0.0376],[447,-0.0376],[448,-0.0279],[454,-0.0279],[455,-0.0285],[456,-0.0285],[457,-0.0292],[458,-0.0292],[459,-0.0298],[463,-0.0298],[464,-0.0305],[466,-0.0305],[467,-0.0311],[469,-0.0311],[470,-0.0314],[471,-0.0314],[472,-0.0317],[473,-0.0317],[474,-0.032],[475,-0.032],[477,-0.0326],[478,-0.0326],[479,-0.0329],[480,-0.0329],[483,-0.0338],[484,-0.0337],[485,-0.035],[490,-0.035],[491,-0.0362],[499,-0.0362],[500,-0.0369],[501,-0.0369],[502,-0.0414],[505,-0.0414],[506,-0.0459],[512,-0.0459],[513,-0.0503],[515,-0.0503],[516,-0.0492],[517,-0.05],[519,-0.05],[520,-0.0508],[523,-0.0508],[524,-0.05],[526,-0.05],[528,-0.0483],[529,-0.0483],[531,-0.0467],[532,-0.0446],[533,-0.0446],[534,-0.0426],[535,-0.0426],[540,-0.033],[547,-0.033],[548,-0.032],[551,-0.032],[552,-0.0304],[554,-0.0304],[555,-0.0287],[556,-0.0287],[557,-0.0271],[559,-0.0271],[560,-0.0254],[561,-0.0254],[563,-0.0221],[564,-0.0217],[565,-0.0204],[567,-0.0187],[570,-0.0147],[571,-0.013],[573,-0.0104],[575,-0.0087],[576,-0.0087],[578,-0.0095],[579,-0.0108],[580,-0.0125],[582,-0.0171],[583,-0.0171],[585,-0.0217],[586,-0.0217],[587,-0.024],[588,-0.024],[589,-0.0263],[593,-0.0263],[594,-0.0285],[595,-0.0285],[596,-0.0273],[599,-0.0273],[600,-0.0303],[606,-0.0303],[607,-0.0332],[610,-0.0332],[611,-0.0362],[612,-0.0356],[613,-0.0356],[614,-0.0392],[617,-0.0392],[618,-0.0427],[619,-0.0427],[620,-0.0463],[621,-0.0463],[622,-0.0499],[623,-0.0499],[624,-0.0535],[625,-0.0535],[627,-0.0606],[628,-0.0612],[629,-0.0653],[631,-0.0693],[632,-0.0733],[634,-0.0773],[676,-0.0821],[681,-0.0876],[686,-0.0896],[687,-0.0905],[688,-0.0919],[689,-0.0938],[690,-0.0961],[691,-0.0988],[693,-0.1048],[694,-0.1116],[695,-0.1116],[696,-0.1184],[697,-0.1184],[699,-0.132],[700,-0.132],[701,-0.1387],[702,-0.1387],[703,-0.1455],[705,-0.1455],[706,-0.1523],[708,-0.1523],[709,-0.1516],[710,-0.1531],[712,-0.1531],[713,-0.1546],[724,-0.1546],[725,-0.1561],[726,-0.1536],[734,-0.1536],[735,-0.151],[738,-0.151],[739,-0.1484],[740,-0.1484],[741,-0.1491],[742,-0.1491],[743,-0.1459],[747,-0.1459],[748,-0.1427],[752,-0.1427],[753,-0.1395],[757,-0.1397],[759,-0.1375],[770,-0.0332],[771,-0.0362],[772,-0.0356],[773,-0.0356],[774,-0.0392],[777,-0.0392],[778,-0.0427],[779,-0.0427],[780,-0.0463],[781,-0.0463],[782,-0.0499],[783,-0.0499],[784,-0.0535],[785,-0.0535],[787,-0.0606],[788,-0.0612],[789,-0.0653],[791,-0.0693],[792,-0.0733],[799,-0.0874],[800,-0.0874],[801,-0.0894],[803,-0.0894],[837,-0.0821],[887,-0.0793],[891,-0.0874],[892,-0.0874],[893,-0.0894],[895,-0.0894],[929,-0.0821]]
       }
       spine_02_ctrl.Location.Y {
        type=MovieSceneScriptingFloatChannel
        default=0
        thinned_from=739
        keys [[-2,-0.2992],[8,-0.3075],[10,-0.3102],[12,-0.3143],[16,-0.3257],[21,-0.3417],[22,-0.3419],[23,-0.345],[29,-0.3457],[30,-0.3426],[33,-0.3434],[34,-0.3373],[38,-0.3374],[39,-0.3312],[43,-0.3313],[44,-0.3252],[47,-0.3249],[48,-0.3232],[50,-0.3232],[51,-0.3215],[54,-0.3213],[55,-0.3196],[62,-0.3191],[64,-0.3194],[65,-0.3208],[70,-0.3205],[71,-0.3222],[75,-0.3219],[76,-0.3236],[82,-0.323],[85,-0.3253],[87,-0.3264],[89,-0.3263],[91,-0.3245],[95,-0.3183],[97,-0.3123],[98,-0.312],[101,-0.3029],[102,-0.3029],[103,-0.2998],[104,-0.2998],[105,-0.2966],[110,-0.2963],[112,-0.2992],[114,-0.2992],[115,-0.3012],[116,-0.3012],[118,-0.3049],[119,-0.3049],[120,-0.3069],[121,-0.3072],[122,-0.3089],[123,-0.3092],[124,-0.3112],[125,-0.3112],[126,-0.3098],[128,-0.3086],[129,-0.3067],[132,-0.3041],[133,-0.3043],[134,-0.3032],[135,-0.3032],[136,-0.3044],[137,-0.3044],[141,-0.3083],[142,-0.3102],[143,-0.3106],[148,-0.3165],[149,-0.3189],[155,-0.3261],[156,-0.3261],[158,-0.3285],[161,-0.3285],[162,-0.3321],[166,-0.3321],[167,-0.3358],[171,-0.3359],[172,-0.3397],[174,-0.3397],[175,-0.3407],[181,-0.3409],[182,-0.3366],[185,-0.3369],[186,-0.3326],[188,-0.3326],[189,-0.3286],[190,-0.3286],[191,-0.3277],[192,-0.3255],[193,-0.3255],[194,-0.3233],[195,-0.3233],[196,-0.3211],[197,-0.3211],[198,-0.3188],[199,-0.3188],[200,-0.3166],[202,-0.3166],[203,-0.3144],[204,-0.3144],[205,-0.3122],[209,-0.3119],[210,-0.3101],[213,-0.3101],[214,-0.3118],[215,-0.3118],[216,-0.3135],[217,-0.3135],[219,-0.3169],[220,-0.3169],[221,-0.3204],[222,-0.3221],[223,-0.3252],[224,-0.3274],[225,-0.3272],[226,-0.3294],[229,-0.3292],[230,-0.327],[231,-0.3269],[234,-0.3202],[235,-0.32],[238,-0.3132],[239,-0.3118],[240,-0.3092],[241,-0.3072],[242,-0.3045],[246,-0.2964],[248,-0.2936],[251,-0.2913],[252,-0.2912],[253,-0.2904],[255,-0.2917],[261,-0.2911],[262,-0.2893],[263,-0.2893],[264,-0.2875],[265,-0.2875],[270,-0.279],[272,-0.2785],[273,-0.2737],[277,-0.2738],[278,-0.2692],[282,-0.2692],[283,-0.2644],[286,-0.2646],[287,-0.2639],[291,-0.2639],[292,-0.2652],[294,-0.2651],[295,-0.2663],[296,-0.2663],[298,-0.2687],[299,-0.2686],[303,-0.2738],[307,-0.2773],[308,-0.2791],[309,-0.28],[323,-0.2992],[333,-0.3075],[335,-0.3102],[337,-0.3143],[341,-0.3257],[346,-0.3417],[347,-0.3419],[348,-0.345],[354,-0.3457],[355,-0.3426],[358,-0.3434],[359,-0.3373],[363,-0.3374],[364,-0.3312],[368,-0.3313],[369,-0.3252],[372,-0.3249],[373,-0.3232],[375,-0.3232],[376,-0.3215],[379,-0.3213],[380,-0.3196],[387,-0.3191],[389,-0.3194],[390,-0.3208],[395,-0.3205],[396,-0.3222],[400,-0.3219],[401,-0.3236],[407,-0.323],[410,-0.3253],[412,-0.3264],[414,-0.3263],[416,-0.3245],[420,-0.3183],[422,-0.3123],[423,-0.312],[426,-0.3029],[427,-0.3029],[428,-0.2998],[429,-0.2998],[430,-0.2966],[435,-0.2963],[437,-0.2992],[439,-0.2992],[440,-0.3012],[441,-0.3012],[443,-0.3049],[444,-0.3049],[445,-0.3069],[446,-0.3072],[447,-0.3089],[448,-0.3092],[449,-0.3112],[450,-0.3112],[451,-0.3098],[453,-0.3086],[454,-0.3067],[457,-0.3041],[458,-0.3043],[459,-0.3032],[460,-0.3032],[461,-0.3044],[462,-0.3044],[466,-0.3083],[467,-0.3102],[468,-0.3106],[473,-0.3165],[474,-0.3189],[480,-0.3261],[481,-0.3261],[483,-0.3285],[486,-0.3285],[487,-0.3321],[491,-0.3321],[492,-0.3358],[496,-0.3359],[497,-0.3397],[499,-0.3397],[500,-0.3407],[506,-0.3409],[507,-0.3366],[510,-0.3369],[511,-0.3326],[513,-0.3326],[514,-0.3286],[515,-0.3286],[516,-0.3277],[517,-0.3255],[518,-0.3255],[519,-0.3233],[520,-0.3233],[521,-0.3211],[522,-0.3211],[523,-0.3188],[524,-0.3188],[525,-0.3166],[527,-0.3166],[528,-0.3144],[529,-0.3144],[530,-0.3122],[534,-0.3119],[535,-0.3101],[538,-0.3101],[539,-0.3118],[540,-0.3118],[541,-0.3135],[542,-0.3135],[544,-0.3169],[545,-0.3169],[546,-0.3204],[547,-0.3221],[548,-0.3252],[549,-0.3274],[550,-0.3272],[551,-0.3294],[554,-0.3292],[555,-0.327],[556,-0.3269],[559,-0.3202],[560,-0.32],[563,-0.3132],[564,-0.3118],[565,-0.3092],[566,-0.3072],[567,-0.3045],[571,-0.2964],[573,-0.2936],[576,-0.2913],[577,-0.2912],[578,-0.2904],[580,-0.2917],[586,-0.2911],[587,-0.2893],[588,-0.2893],[589,-0.2875],[590,-0.2875],[595,-0.279],[597,-0.2785],[598,-0.2737],[602,-0.2738],[603,-0.2692],[607,-0.2692],[608,-0.2644],[611,-0.2646],[612,-0.2639],[616,-0.2639],[617,-0.2652],[619,-0.2651],[620,-0.2663],[621,-0.2663],[623,-0.2687],[624,-0.2686],[628,-0.2738],[632,-0.2773],[633,-0.2791],[634,-0.28],[676,-0.2992],[686,-0.3075],[688,-0.3102],[690,-0.3143],[694,-0.3257],[699,-0.3417],[700,-0.3419],[701,-0.345],[707,-0.3457],[708,-0.3426],[711,-0.3434],[712,-0.3373],[716,-0.3374],[717,-0.3312],[721,-0.3313],[722,-0.3252],[725,-0.3249],[726,-0.3232],[728,-0.3232],[729,-0.3215],[732,-0.3213],[733,-0.3196],[740,-0.3191],[742,-0.3194],[743,-0.3208],[748,-0.3205],[749,-0.3222],[753,-0.3219],[754,-0.3236],[759,-0.3231],[770,-0.2646],[771,-0.2646],[772,-0.2639],[776,-0.2639],[777,-0.2652],[779,-0.2651],[780,-0.2663],[781,-0.2663],[783,-0.2687],[784,-0.2686],[788,-0.2738],[792,-0.2773],[793,-0.2791],[799,-0.2844],[800,-0.2844],[801,-0.2853],[802,-0.2853],[803,-0.2862],[837,-0.2992],[887,-0.2809],[891,-0.2844],[892,-0.2844],[893,-0.2853],[894,-0.2853],[895,-0.2862],[929,-0.2992]]
       }
       spine_02_ctrl.Location.Z {
        type=MovieSceneScriptingFloatChannel
        default=0
        thinned_from=739
        keys [[-2,-0.0281],[3,-0.0293],[7,-0.0295],[10,-0.0289],[12,-0.028],[15,-0.0261],[16,-0.0263],[17,-0.0236],[18,-0.0238],[20,-0.0183],[21,-0.0185],[22,-0.0156],[23,-0.0158],[24,-0.0128],[26,-0.0128],[27,-0.0098],[28,-0.0098],[29,-0.0068],[30,-0.0066],[31,-0.0066],[32,-0.0054],[33,-0.0054],[34,-0.005],[35,-0.0038],[38,-0.0038],[39,-0.0033],[40,-0.0021],[43,-0.0021],[44,-0.0016],[51,-0.0013],[52,-0.0043],[56,-0.0041],[57,-0.0071],[60,-0.0071],[61,-0.0101],[64,-0.0101],[65,-0.0139],[68,-0.0139],[69,-0.0176],[74,-0.0177],[75,-0.0214],[78,-0.0215],[79,-0.0221],[81,-0.022],[82,-0.023],[83,-0.0231],[91,-0.0307],[92,-0.0325],[93,-0.0334],[94,-0.0351],[95,-0.0357],[96,-0.0391],[97,-0.0389],[98,-0.0425],[99,-0.0422],[101,-0.049],[102,-0.049],[103,-0.0524],[104,-0.0524],[105,-0.0557],[107,-0.0557],[108,-0.0593],[114,-0.0596],[115,-0.0561],[117,-0.0562],[118,-0.0527],[119,-0.0527],[121,-0.0455],[122,-0.0456],[124,-0.0385],[125,-0.0385],[126,-0.0347],[127,-0.0332],[133,-0.0218],[135,-0.0217],[136,-0.02],[138,-0.02],[139,-0.0219],[141,-0.0221],[142,-0.0241],[143,-0.0237],[145,-0.0256],[146,-0.0275],[155,-0.0364],[156,-0.0364],[158,-0.0383],[159,-0.0383],[160,-0.0393],[161,-0.0393],[162,-0.0405],[165,-0.0405],[166,-0.0415],[167,-0.0418],[169,-0.0418],[170,-0.0408],[171,-0.0408],[172,-0.0411],[173,-0.0411],[174,-0.0402],[178,-0.04],[179,-0.0371],[181,-0.0371],[182,-0.0367],[183,-0.0367],[184,-0.0338],[185,-0.0338],[186,-0.0334],[188,-0.0334],[189,-0.0301],[190,-0.0301],[191,-0.0304],[193,-0.0291],[194,-0.0289],[195,-0.0284],[196,-0.0282],[197,-0.0276],[199,-0.0274],[200,-0.0267],[202,-0.0267],[203,-0.0265],[204,-0.026],[210,-0.0255],[211,-0.0259],[213,-0.0259],[214,-0.0264],[215,-0.0264],[216,-0.027],[218,-0.0271],[219,-0.0277],[220,-0.0277],[221,-0.0279],[222,-0.0285],[224,-0.0289],[225,-0.0312],[227,-0.0314],[228,-0.0336],[230,-0.0334],[231,-0.0356],[232,-0.0355],[233,-0.0375],[234,-0.0374],[236,-0.0416],[237,-0.0415],[238,-0.0435],[251,-0.063],[254,-0.0678],[255,-0.0686],[260,-0.0757],[261,-0.0757],[262,-0.077],[263,-0.077],[264,-0.0783],[270,-0.0776],[271,-0.0769],[272,-0.0769],[273,-0.0765],[274,-0.0751],[277,-0.0751],[278,-0.0733],[282,-0.0733],[283,-0.0729],[284,-0.0729],[285,-0.0715],[286,-0.0715],[287,-0.0725],[292,-0.0726],[293,-0.0736],[294,-0.0736],[295,-0.0748],[296,-0.0748],[297,-0.076],[298,-0.0761],[299,-0.0771],[300,-0.0772],[302,-0.0796],[306,-0.0822],[307,-0.0834],[309,-0.0848],[323,-0.0281],[328,-0.0293],[332,-0.0295],[335,-0.0289],[337,-0.028],[340,-0.0261],[341,-0.0263],[342,-0.0236],[343,-0.0238],[345,-0.0183],[346,-0.0185],[347,-0.0156],[348,-0.0158],[349,-0.0128],[351,-0.0128],[352,-0.0098],[353,-0.0098],[354,-0.0068],[355,-0.0066],[356,-0.0066],[357,-0.0054],[358,-0.0054],[359,-0.005],[360,-0.0038],[363,-0.0038],[364,-0.0033],[365,-0.0021],[368,-0.0021],[369,-0.0016],[376,-0.0013],[377,-0.0043],[381,-0.0041],[382,-0.0071],[385,-0.0071],[386,-0.0101],[389,-0.0101],[390,-0.0139],[393,-0.0139],[394,-0.0176],[399,-0.0177],[400,-0.0214],[403,-0.0215],[404,-0.0221],[406,-0.022],[407,-0.023],[408,-0.0231],[416,-0.0307],[417,-0.0325],[418,-0.0334],[419,-0.0351],[420,-0.0357],[421,-0.0391],[422,-0.0389],[423,-0.0425],[424,-0.0422],[426,-0.049],[427,-0.049],[428,-0.0524],[429,-0.0524],[430,-0.0557],[432,-0.0557],[433,-0.0593],[439,-0.0596],[440,-0.0561],[442,-0.0562],[443,-0.0527],[444,-0.0527],[446,-0.0455],[447,-0.0456],[449,-0.0385],[450,-0.0385],[451,-0.0347],[452,-0.0332],[458,-0.0218],[460,-0.0217],[461,-0.02],[463,-0.02],[464,-0.0219],[466,-0.0221],[467,-0.0241],[468,-0.0237],[470,-0.0256],[471,-0.0275],[480,-0.0364],[481,-0.0364],[483,-0.0383],[484,-0.0383],[485,-0.0393],[486,-0.0393],[487,-0.0405],[490,-0.0405],[491,-0.0415],[492,-0.0418],[494,-0.0418],[495,-0.0408],[496,-0.0408],[497,-0.0411],[498,-0.0411],[499,-0.0402],[503,-0.04],[504,-0.0371],[506,-0.0371],[507,-0.0367],[508,-0.0367],[509,-0.0338],[510,-0.0338],[511,-0.0334],[513,-0.0334],[514,-0.0301],[515,-0.0301],[516,-0.0304],[518,-0.0291],[519,-0.0289],[520,-0.0284],[521,-0.0282],[522,-0.0276],[524,-0.0274],[525,-0.0267],[527,-0.0267],[528,-0.0265],[529,-0.026],[535,-0.0255],[536,-0.0259],[538,-0.0259],[539,-0.0264],[540,-0.0264],[541,-0.027],[543,-0.0271],[544,-0.0277],[545,-0.0277],[546,-0.0279],[547,-0.0285],[549,-0.0289],[550,-0.0312],[552,-0.0314],[553,-0.0336],[555,-0.0334],[556,-0.0356],[557,-0.0355],[558,-0.0375],[559,-0.0374],[561,-0.0416],[562,-0.0415],[563,-0.0435],[576,-0.063],[579,-0.0678],[580,-0.0686],[585,-0.0757],[586,-0.0757],[587,-0.077],[588,-0.077],[589,-0.0783],[595,-0.0776],[596,-0.0769],[597,-0.0769],[598,-0.0765],[599,-0.0751],[602,-0.0751],[603,-0.0733],[607,-0.0733],[608,-0.0729],[609,-0.0729],[610,-0.0715],[611,-0.0715],[612,-0.0725],[617,-0.0726],[618,-0.0736],[619,-0.0736],[620,-0.0748],[621,-0.0748],[622,-0.076],[623,-0.0761],[624,-0.0771],[625,-0.0772],[627,-0.0796],[631,-0.0822],[632,-0.0834],[634,-0.0848],[676,-0.0281],[681,-0.0293],[685,-0.0295],[688,-0.0289],[690,-0.028],[693,-0.0261],[694,-0.0263],[695,-0.0236],[696,-0.0238],[698,-0.0183],[699,-0.0185],[700,-0.0156],[701,-0.0158],[702,-0.0128],[704,-0.0128],[705,-0.0098],[706,-0.0098],[707,-0.0068],[708,-0.0066],[709,-0.0066],[710,-0.0054],[711,-0.0054],[712,-0.005],[713,-0.0038],[716,-0.0038],[717,-0.0033],[718,-0.0021],[721,-0.0021],[722,-0.0016],[729,-0.0013],[730,-0.0043],[734,-0.0041],[735,-0.0071],[738,-0.0071],[739,-0.0101],[742,-0.0101],[743,-0.0139],[746,-0.0139],[747,-0.0176],[752,-0.0177],[753,-0.0214],[755,-0.0215],[757,-0.0221],[759,-0.022],[770,-0.0715],[771,-0.0715],[772,-0.0725],[777,-0.0726],[778,-0.0736],[779,-0.0736],[780,-0.0748],[781,-0.0748],[782,-0.076],[783,-0.0761],[784,-0.0771],[785,-0.0772],[787,-0.0796],[791,-0.0822],[792,-0.0834],[798,-0.0873],[799,-0.0874],[801,-0.0886],[802,-0.0886],[803,-0.0893],[837,-0.0281],[887,-0.0854],[890,-0.0873],[891,-0.0874],[893,-0.0886],[894,-0.0886],[895,-0.0893],[929,-0.0281]]
       }
       spine_02_ctrl.Rotation.X {
        type=MovieSceneScriptingFloatChannel
        default=0
        thinned_from=543
        keys [[-2,-4.6275],[56,-5.2409],[75,-5.1036],[83,-5.0786],[85,-5.0864],[87,-5.0706],[89,-5.0286],[91,-4.959],[94,-4.815],[97,-4.6369],[102,-4.2941],[104,-4.1767],[105,-4.1281],[106,-4.0879],[108,-4.0333],[110,-3.9967],[112,-3.9782],[114,-3.9797],[115,-3.9936],[116,-4.0162],[118,-4.0854],[120,-4.1809],[124,-4.3934],[127,-4.5172],[129,-4.5757],[131,-4.6164],[147,-4.7821],[164,-4.8326],[178,-5.0491],[182,-5.0691],[189,-5.0203],[210,-4.7428],[214,-4.7181],[222,-4.7691],[224,-4.7625],[225,-4.7475],[226,-4.7211],[228,-4.6393],[239,-4.0562],[244,-3.7511],[247,-3.5948],[249,-3.5114],[251,-3.4481],[253,-3.4047],[261,-3.3177],[264,-3.2633],[270,-3.1129],[273,-3.0548],[284,-2.9167],[291,-2.874],[296,-2.8774],[309,-2.9877],[352,-4.921],[381,-5.2409],[400,-5.1036],[408,-5.0786],[410,-5.0864],[412,-5.0706],[414,-5.0286],[416,-4.959],[419,-4.815],[422,-4.6369],[427,-4.2941],[429,-4.1767],[430,-4.1281],[431,-4.0879],[433,-4.0333],[435,-3.9967],[437,-3.9782],[439,-3.9797],[440,-3.9936],[441,-4.0162],[443,-4.0854],[445,-4.1809],[449,-4.3934],[452,-4.5172],[454,-4.5757],[456,-4.6164],[472,-4.7821],[489,-4.8326],[503,-5.0491],[507,-5.0691],[514,-5.0203],[535,-4.7428],[539,-4.7181],[547,-4.7691],[549,-4.7625],[550,-4.7475],[551,-4.7211],[553,-4.6393],[564,-4.0562],[569,-3.7511],[572,-3.5948],[574,-3.5114],[576,-3.4481],[578,-3.4047],[586,-3.3177],[589,-3.2633],[595,-3.1129],[598,-3.0548],[609,-2.9167],[616,-2.874],[621,-2.8774],[634,-2.9877],[676,2.1435],[734,1.53],[740,1.5913],[745,1.6328],[796,2.8864],[804,2.8206],[837,2.1435],[887,-2.9982],[895,-3.0641],[929,-4.6275]]
       }
       spine_02_ctrl.Rotation.Y {
        type=MovieSceneScriptingFloatChannel
        default=0
        thinned_from=677
        keys [[-2,1.9371],[6,2.0565],[13,2.3687],[15,2.5162],[20,2.9379],[21,3.0179],[22,3.0885],[23,3.1501],[24,3.2009],[26,3.2771],[28,3.3322],[30,3.3676],[35,3.4016],[49,3.3769],[56,3.3252],[81,3.0358],[83,3.0214],[84,3.0219],[85,3.0306],[87,3.0704],[88,3.0986],[92,3.2432],[93,3.2686],[94,3.283],[95,3.2841],[96,3.2662],[97,3.2278],[98,3.1721],[99,3.1038],[100,3.021],[101,2.9277],[102,2.8206],[103,2.7026],[104,2.5733],[105,2.4365],[107,2.1459],[110,1.6992],[111,1.5549],[112,1.4148],[113,1.2777],[114,1.1442],[115,1.0156],[116,0.8955],[117,0.7816],[118,0.6779],[119,0.5835],[120,0.5001],[121,0.427],[122,0.3623],[123,0.3079],[124,0.2634],[125,0.2288],[126,0.2026],[127,0.1835],[128,0.1713],[129,0.1624],[132,0.1404],[134,0.1294],[135,0.1248],[136,0.1218],[137,0.1213],[138,0.1214],[140,0.1268],[141,0.1299],[142,0.1344],[143,0.1395],[144,0.146],[145,0.1516],[146,0.1586],[147,0.1653],[148,0.1723],[149,0.179],[156,0.2214],[157,0.2264],[159,0.2383],[161,0.2514],[162,0.2571],[163,0.2599],[164,0.2618],[165,0.2628],[166,0.2629],[167,0.2619],[169,0.2582],[170,0.2552],[171,0.2536],[172,0.2543],[173,0.256],[174,0.2586],[175,0.265],[176,0.2745],[177,0.2868],[178,0.3005],[179,0.3157],[180,0.3297],[181,0.3411],[182,0.3548],[184,0.3763],[186,0.392],[187,0.4017],[188,0.4084],[189,0.4167],[191,0.4315],[193,0.4435],[195,0.4529],[197,0.4594],[199,0.4619],[200,0.462],[201,0.4604],[203,0.454],[204,0.4485],[205,0.4418],[206,0.4338],[207,0.4239],[208,0.4127],[209,0.3998],[210,0.3853],[211,0.369],[212,0.3492],[213,0.3267],[214,0.3098],[215,0.3035],[216,0.3082],[217,0.3226],[218,0.342],[219,0.3632],[220,0.3827],[221,0.3998],[222,0.4147],[226,0.4685],[227,0.4832],[228,0.4991],[229,0.5166],[230,0.5367],[233,0.6063],[234,0.6279],[235,0.6457],[236,0.6601],[237,0.6719],[239,0.6902],[240,0.698],[241,0.7072],[242,0.7143],[243,0.7168],[244,0.7159],[245,0.7122],[247,0.6982],[248,0.6935],[249,0.6927],[250,0.6989],[251,0.7105],[252,0.7303],[253,0.7575],[254,0.7924],[255,0.8332],[257,0.922],[260,1.0635],[261,1.1062],[262,1.1442],[263,1.1756],[264,1.2031],[265,1.2237],[266,1.2399],[267,1.2498],[271,1.2771],[273,1.2961],[278,1.3398],[279,1.3462],[281,1.3674],[282,1.3821],[283,1.394],[286,1.4386],[287,1.4575],[288,1.4726],[290,1.5101],[292,1.5537],[295,1.6281],[297,1.6862],[299,1.7529],[301,1.8245],[308,2.0979],[309,2.1329],[325,2.2772],[339,2.4405],[340,2.5162],[345,2.9379],[346,3.0179],[347,3.0885],[348,3.1501],[349,3.2009],[351,3.2771],[353,3.3322],[355,3.3676],[360,3.4016],[374,3.3769],[381,3.3252],[406,3.0358],[408,3.0214],[409,3.0219],[410,3.0306],[412,3.0704],[413,3.0986],[417,3.2432],[418,3.2686],[419,3.283],[420,3.2841],[421,3.2662],[422,3.2278],[423,3.1721],[424,3.1038],[425,3.021],[426,2.9277],[427,2.8206],[428,2.7026],[429,2.5733],[430,2.4365],[432,2.1459],[435,1.6992],[436,1.5549],[437,1.4148],[438,1.2777],[439,1.1442],[440,1.0156],[441,0.8955],[442,0.7816],[443,0.6779],[444,0.5835],[445,0.5001],[446,0.427],[447,0.3623],[448,0.3079],[449,0.2634],[450,0.2288],[451,0.2026],[452,0.1835],[453,0.1713],[454,0.1624],[457,0.1404],[459,0.1294],[460,0.1248],[461,0.1218],[462,0.1213],[463,0.1214],[465,0.1268],[466,0.1299],[467,0.1344],[468,0.1395],[469,0.146],[470,0.1516],[471,0.1586],[472,0.1653],[473,0.1723],[474,0.179],[481,0.2214],[482,0.2264],[484,0.2383],[486,0.2514],[487,0.2571],[488,0.2599],[489,0.2618],[490,0.2628],[491,0.2629],[492,0.2619],[494,0.2582],[495,0.2552],[496,0.2536],[497,0.2543],[498,0.256],[499,0.2586],[500,0.265],[501,0.2745],[502,0.2868],[503,0.3005],[504,0.3157],[505,0.3297],[506,0.3411],[507,0.3548],[509,0.3763],[511,0.392],[512,0.4017],[513,0.4084],[514,0.4167],[516,0.4315],[518,0.4435],[520,0.4529],[522,0.4594],[524,0.4619],[525,0.462],[526,0.4604],[528,0.454],[529,0.4485],[530,0.4418],[531,0.4338],[532,0.4239],[533,0.4127],[534,0.3998],[535,0.3853],[536,0.369],[537,0.3492],[538,0.3267],[539,0.3098],[540,0.3035],[541,0.3082],[542,0.3226],[543,0.342],[544,0.3632],[545,0.3827],[546,0.3998],[547,0.4147],[551,0.4685],[552,0.4832],[553,0.4991],[554,0.5166],[555,0.5367],[558,0.6063],[559,0.6279],[560,0.6457],[561,0.6601],[562,0.6719],[564,0.6902],[565,0.698],[566,0.7072],[567,0.7143],[568,0.7168],[569,0.7159],[570,0.7122],[572,0.6982],[573,0.6935],[574,0.6927],[575,0.6989],[576,0.7105],[577,0.7303],[578,0.7575],[579,0.7924],[580,0.8332],[582,0.922],[585,1.0635],[586,1.1062],[587,1.1442],[588,1.1756],[589,1.2031],[590,1.2237],[591,1.2399],[592,1.2498],[596,1.2771],[598,1.2961],[603,1.3398],[604,1.3462],[606,1.3674],[607,1.3821],[608,1.394],[611,1.4386],[612,1.4575],[613,1.4726],[615,1.5101],[617,1.5537],[620,1.6281],[622,1.6862],[624,1.7529],[626,1.8245],[633,2.0979],[634,2.1329],[676,1.9371],[684,2.0565],[691,2.3687],[693,2.5162],[698,2.9379],[699,3.0179],[700,3.0885],[701,3.1501],[702,3.2009],[704,3.2771],[706,3.3322],[708,3.3676],[713,3.4016],[727,3.3769],[734,3.3252],[755,3.0731],[759,3.0358],[803,2.3253],[837,1.9371],[887,2.1668],[889,2.2251],[891,2.2706],[895,2.3253],[929,1.9371]]
       }
       spine_02_ctrl.Rotation.Z {
        type=MovieSceneScriptingFloatChannel
        default=0
        thinned_from=628
        keys [[-2,0.0892],[3,-0.074],[4,-0.1197],[5,-0.1697],[6,-0.2244],[7,-0.2804],[8,-0.3417],[9,-0.4079],[10,-0.4776],[11,-0.5505],[12,-0.6266],[14,-0.7846],[15,-0.8659],[16,-0.9495],[18,-1.113],[20,-1.2703],[21,-1.3444],[22,-1.4155],[23,-1.4805],[24,-1.539],[25,-1.5927],[26,-1.6398],[27,-1.6781],[28,-1.7095],[29,-1.7339],[30,-1.7529],[31,-1.7668],[36,-1.8081],[44,-1.8455],[51,-1.8559],[57,-1.8456],[63,-1.8227],[71,-1.7858],[74,-1.7755],[83,-1.7957],[84,-1.8365],[85,-1.9028],[86,-1.9891],[87,-2.0894],[88,-2.212],[89,-2.3517],[90,-2.5077],[92,-2.8347],[93,-2.991],[94,-3.1376],[95,-3.2617],[96,-3.361],[97,-3.4275],[98,-3.4676],[99,-3.483],[100,-3.4761],[101,-3.4466],[102,-3.3938],[103,-3.3205],[104,-3.2232],[105,-3.1108],[107,-2.8626],[110,-2.4699],[111,-2.3442],[114,-1.9961],[115,-1.8862],[116,-1.7817],[117,-1.6892],[118,-1.6126],[119,-1.5565],[120,-1.5212],[121,-1.5048],[122,-1.5007],[123,-1.5106],[124,-1.538],[125,-1.5779],[126,-1.6299],[129,-1.8065],[130,-1.8606],[131,-1.9107],[132,-1.953],[134,-2.0224],[136,-2.0705],[137,-2.0862],[138,-2.0902],[152,-2.1002],[170,-2.0493],[171,-2.0497],[172,-2.0546],[174,-2.0752],[175,-2.0907],[177,-2.1432],[180,-2.2346],[182,-2.2798],[183,-2.2965],[187,-2.3146],[191,-2.3133],[193,-2.3042],[203,-1.9744],[205,-1.8853],[206,-1.836],[207,-1.7804],[208,-1.7163],[209,-1.6413],[210,-1.556],[211,-1.4634],[212,-1.3597],[213,-1.2412],[214,-1.1017],[215,-0.9421],[216,-0.7572],[217,-0.5566],[218,-0.3468],[219,-0.1268],[220,0.0922],[221,0.3122],[222,0.531],[223,0.7441],[224,0.94],[225,1.1221],[226,1.2857],[227,1.4257],[228,1.541],[229,1.634],[230,1.7129],[231,1.7832],[232,1.837],[237,1.9294],[243,1.9405],[244,1.9273],[245,1.9064],[246,1.8805],[247,1.8491],[249,1.7718],[251,1.6752],[253,1.5659],[255,1.4489],[257,1.3375],[259,1.235],[261,1.1519],[262,1.1213],[263,1.0938],[264,1.0757],[265,1.0666],[266,1.0626],[269,1.0941],[270,1.1061],[274,1.1396],[279,1.1683],[280,1.1715],[281,1.1772],[283,1.1831],[290,1.2032],[292,1.1959],[294,1.1789],[295,1.1671],[296,1.1505],[297,1.1247],[298,1.0963],[299,1.0614],[300,1.0216],[301,0.9776],[302,0.9308],[304,0.8263],[306,0.7127],[309,0.5383],[323,0.0892],[328,-0.074],[329,-0.1197],[330,-0.1697],[331,-0.2244],[332,-0.2804],[333,-0.3417],[334,-0.4079],[335,-0.4776],[336,-0.5505],[337,-0.6266],[339,-0.7846],[340,-0.8659],[341,-0.9495],[343,-1.113],[345,-1.2703],[346,-1.3444],[347,-1.4155],[348,-1.4805],[349,-1.539],[350,-1.5927],[351,-1.6398],[352,-1.6781],[353,-1.7095],[354,-1.7339],[355,-1.7529],[356,-1.7668],[361,-1.8081],[369,-1.8455],[376,-1.8559],[382,-1.8456],[388,-1.8227],[396,-1.7858],[399,-1.7755],[408,-1.7957],[409,-1.8365],[410,-1.9028],[411,-1.9891],[412,-2.0894],[413,-2.212],[414,-2.3517],[415,-2.5077],[417,-2.8347],[418,-2.991],[419,-3.1376],[420,-3.2617],[421,-3.361],[422,-3.4275],[423,-3.4676],[424,-3.483],[425,-3.4761],[426,-3.4466],[427,-3.3938],[428,-3.3205],[429,-3.2232],[430,-3.1108],[432,-2.8626],[435,-2.4699],[436,-2.3442],[439,-1.9961],[440,-1.8862],[441,-1.7817],[442,-1.6892],[443,-1.6126],[444,-1.5565],[445,-1.5212],[446,-1.5048],[447,-1.5007],[448,-1.5106],[449,-1.538],[450,-1.5779],[451,-1.6299],[454,-1.8065],[455,-1.8606],[456,-1.9107],[457,-1.953],[459,-2.0224],[461,-2.0705],[462,-2.0862],[463,-2.0902],[477,-2.1002],[495,-2.0493],[496,-2.0497],[497,-2.0546],[499,-2.0752],[500,-2.0907],[502,-2.1432],[505,-2.2346],[507,-2.2798],[508,-2.2965],[512,-2.3146],[516,-2.3133],[518,-2.3042],[528,-1.9744],[530,-1.8853],[531,-1.836],[532,-1.7804],[533,-1.7163],[534,-1.6413],[535,-1.556],[536,-1.4634],[537,-1.3597],[538,-1.2412],[539,-1.1017],[540,-0.9421],[541,-0.7572],[542,-0.5566],[543,-0.3468],[544,-0.1268],[545,0.0922],[546,0.3122],[547,0.531],[548,0.7441],[549,0.94],[550,1.1221],[551,1.2857],[552,1.4257],[553,1.541],[554,1.634],[555,1.7129],[556,1.7832],[557,1.837],[562,1.9294],[568,1.9405],[569,1.9273],[570,1.9064],[571,1.8805],[572,1.8491],[574,1.7718],[576,1.6752],[578,1.5659],[580,1.4489],[582,1.3375],[584,1.235],[586,1.1519],[587,1.1213],[588,1.0938],[589,1.0757],[590,1.0666],[591,1.0626],[594,1.0941],[595,1.1061],[599,1.1396],[604,1.1683],[605,1.1715],[606,1.1772],[608,1.1831],[615,1.2032],[617,1.1959],[619,1.1789],[620,1.1671],[621,1.1505],[622,1.1247],[623,1.0963],[624,1.0614],[625,1.0216],[626,0.9776],[627,0.9308],[629,0.8263],[631,0.7127],[634,0.5383],[676,0.0892],[681,-0.074],[682,-0.1197],[683,-0.1697],[684,-0.2244],[685,-0.2804],[686,-0.3417],[687,-0.4079],[688,-0.4776],[689,-0.5505],[690,-0.6266],[692,-0.7846],[693,-0.8659],[694,-0.9495],[696,-1.113],[698,-1.2703],[699,-1.3444],[700,-1.4155],[701,-1.4805],[702,-1.539],[703,-1.5927],[704,-1.6398],[705,-1.6781],[706,-1.7095],[707,-1.7339],[708,-1.7529],[709,-1.7668],[714,-1.8081],[722,-1.8455],[729,-1.8559],[735,-1.8456],[741,-1.8227],[749,-1.7858],[755,-1.7682],[759,-1.7626],[803,0.0829],[837,0.0892],[887,0.4806],[889,0.3662],[890,0.3126],[891,0.2611],[892,0.2117],[893,0.1656],[894,0.1227],[895,0.0829],[929,0.0892]]
       }
       spine_02_ctrl.Scale.X {type=MovieSceneScriptingFloatChannel, default=1, const [1,739]}
       spine_02_ctrl.Scale.Y {type=MovieSceneScriptingFloatChannel, default=1, const [1,739]}
       spine_02_ctrl.Scale.Z {type=MovieSceneScriptingFloatChannel, default=1, const [1,739]}
       spine_03_ctrl.Location.X {
        type=MovieSceneScriptingFloatChannel
        default=0
        thinned_from=735
        keys [[-2,-0.1341],[-1,-0.1344],[6,-0.1423],[9,-0.1423],[10,-0.145],[11,-0.145],[14,-0.1533],[15,-0.1551],[16,-0.16],[17,-0.16],[19,-0.1697],[20,-0.1697],[21,-0.1746],[22,-0.1746],[23,-0.1794],[24,-0.1794],[25,-0.1843],[28,-0.1843],[29,-0.1892],[31,-0.1891],[40,-0.1939],[41,-0.1939],[43,-0.195],[44,-0.195],[45,-0.1955],[46,-0.1955],[47,-0.1966],[52,-0.1966],[53,-0.1945],[57,-0.1945],[58,-0.1925],[60,-0.1925],[61,-0.1904],[62,-0.1904],[63,-0.1893],[64,-0.1893],[65,-0.1877],[66,-0.1877],[67,-0.186],[68,-0.186],[69,-0.1843],[71,-0.1843],[72,-0.1826],[73,-0.1826],[74,-0.1809],[76,-0.1809],[77,-0.1793],[78,-0.1793],[79,-0.1785],[81,-0.1785],[82,-0.1774],[84,-0.1774],[88,-0.1817],[92,-0.1903],[94,-0.1924],[95,-0.1941],[96,-0.1941],[100,-0.1845],[103,-0.1702],[105,-0.1558],[106,-0.151],[107,-0.1414],[110,-0.1198],[111,-0.113],[117,-0.0703],[121,-0.0513],[122,-0.0489],[123,-0.0441],[124,-0.0418],[125,-0.0418],[126,-0.0394],[127,-0.0386],[129,-0.0381],[131,-0.0381],[137,-0.0388],[139,-0.0395],[143,-0.0406],[148,-0.0429],[149,-0.0429],[153,-0.0447],[154,-0.0447],[158,-0.0465],[159,-0.0468],[160,-0.0475],[161,-0.0475],[162,-0.0482],[163,-0.0482],[164,-0.0489],[167,-0.0489],[168,-0.0496],[172,-0.0496],[173,-0.0503],[174,-0.0503],[175,-0.0511],[176,-0.0511],[178,-0.0556],[179,-0.0556],[180,-0.0578],[181,-0.0578],[182,-0.06],[183,-0.06],[184,-0.0622],[186,-0.0622],[187,-0.0644],[189,-0.0644],[190,-0.0666],[191,-0.0664],[202,-0.0664],[203,-0.0643],[205,-0.0643],[206,-0.0623],[208,-0.0602],[209,-0.0582],[210,-0.0571],[214,-0.0489],[216,-0.0468],[217,-0.0478],[218,-0.0478],[219,-0.0489],[222,-0.0489],[224,-0.0483],[225,-0.0478],[226,-0.0475],[227,-0.0471],[228,-0.0468],[230,-0.0468],[231,-0.0471],[234,-0.0485],[238,-0.0495],[239,-0.0501],[244,-0.0501],[245,-0.0492],[246,-0.0492],[247,-0.0483],[251,-0.0483],[252,-0.0492],[255,-0.0544],[259,-0.0622],[261,-0.0657],[262,-0.067],[263,-0.0682],[264,-0.0691],[265,-0.0698],[266,-0.0703],[271,-0.0709],[273,-0.0709],[274,-0.0718],[277,-0.0718],[278,-0.0727],[280,-0.0727],[281,-0.0736],[282,-0.0736],[284,-0.0754],[285,-0.0754],[287,-0.0771],[288,-0.0787],[289,-0.0787],[291,-0.0817],[292,-0.0817],[299,-0.0924],[300,-0.0955],[302,-0.0985],[303,-0.1013],[304,-0.1046],[305,-0.1046],[307,-0.1114],[308,-0.1114],[309,-0.1147],[323,-0.1341],[324,-0.1344],[331,-0.1423],[334,-0.1423],[335,-0.145],[336,-0.145],[339,-0.1533],[340,-0.1551],[341,-0.16],[342,-0.16],[344,-0.1697],[345,-0.1697],[346,-0.1746],[347,-0.1746],[348,-0.1794],[349,-0.1794],[350,-0.1843],[353,-0.1843],[354,-0.1892],[356,-0.1891],[365,-0.1939],[366,-0.1939],[368,-0.195],[369,-0.195],[370,-0.1955],[371,-0.1955],[372,-0.1966],[377,-0.1966],[378,-0.1945],[382,-0.1945],[383,-0.1925],[385,-0.1925],[386,-0.1904],[387,-0.1904],[388,-0.1893],[389,-0.1893],[390,-0.1877],[391,-0.1877],[392,-0.186],[393,-0.186],[394,-0.1843],[396,-0.1843],[397,-0.1826],[398,-0.1826],[399,-0.1809],[401,-0.1809],[402,-0.1793],[403,-0.1793],[404,-0.1785],[406,-0.1785],[407,-0.1774],[409,-0.1774],[413,-0.1817],[417,-0.1903],[419,-0.1924],[420,-0.1941],[421,-0.1941],[425,-0.1845],[428,-0.1702],[430,-0.1558],[431,-0.151],[432,-0.1414],[435,-0.1198],[436,-0.113],[442,-0.0703],[446,-0.0513],[447,-0.0489],[448,-0.0441],[449,-0.0418],[450,-0.0418],[451,-0.0394],[452,-0.0386],[454,-0.0381],[456,-0.0381],[462,-0.0388],[464,-0.0395],[468,-0.0406],[473,-0.0429],[474,-0.0429],[478,-0.0447],[479,-0.0447],[483,-0.0465],[484,-0.0468],[485,-0.0475],[486,-0.0475],[487,-0.0482],[488,-0.0482],[489,-0.0489],[492,-0.0489],[493,-0.0496],[497,-0.0496],[498,-0.0503],[499,-0.0503],[500,-0.0511],[501,-0.0511],[503,-0.0556],[504,-0.0556],[505,-0.0578],[506,-0.0578],[507,-0.06],[508,-0.06],[509,-0.0622],[511,-0.0622],[512,-0.0644],[514,-0.0644],[515,-0.0666],[516,-0.0664],[527,-0.0664],[528,-0.0643],[530,-0.0643],[531,-0.0623],[533,-0.0602],[534,-0.0582],[535,-0.0571],[539,-0.0489],[541,-0.0468],[542,-0.0478],[543,-0.0478],[544,-0.0489],[547,-0.0489],[549,-0.0483],[550,-0.0478],[551,-0.0475],[552,-0.0471],[553,-0.0468],[555,-0.0468],[556,-0.0471],[559,-0.0485],[563,-0.0495],[564,-0.0501],[569,-0.0501],[570,-0.0492],[571,-0.0492],[572,-0.0483],[576,-0.0483],[577,-0.0492],[580,-0.0544],[584,-0.0622],[586,-0.0657],[587,-0.067],[588,-0.0682],[589,-0.0691],[590,-0.0698],[591,-0.0703],[596,-0.0709],[598,-0.0709],[599,-0.0718],[602,-0.0718],[603,-0.0727],[605,-0.0727],[606,-0.0736],[607,-0.0736],[609,-0.0754],[610,-0.0754],[612,-0.0771],[613,-0.0787],[614,-0.0787],[616,-0.0817],[617,-0.0817],[624,-0.0924],[625,-0.0955],[627,-0.0985],[628,-0.1013],[629,-0.1046],[630,-0.1046],[632,-0.1114],[633,-0.1114],[634,-0.1147],[677,-0.1344],[684,-0.1423],[687,-0.1423],[688,-0.145],[689,-0.145],[692,-0.1533],[693,-0.1551],[694,-0.16],[695,-0.16],[697,-0.1697],[698,-0.1697],[699,-0.1746],[700,-0.1746],[701,-0.1794],[702,-0.1794],[703,-0.1843],[706,-0.1843],[707,-0.1892],[709,-0.1891],[718,-0.1939],[719,-0.1939],[721,-0.195],[722,-0.195],[723,-0.1955],[724,-0.1955],[725,-0.1966],[730,-0.1966],[731,-0.1945],[735,-0.1945],[736,-0.1925],[738,-0.1925],[739,-0.1904],[740,-0.1904],[741,-0.1893],[742,-0.1893],[743,-0.1877],[744,-0.1877],[745,-0.186],[746,-0.186],[747,-0.1843],[749,-0.1843],[750,-0.1826],[751,-0.1826],[752,-0.1809],[754,-0.1809],[755,-0.1793],[757,-0.1785],[759,-0.1785],[770,-0.0754],[772,-0.0771],[773,-0.0787],[774,-0.0787],[776,-0.0817],[777,-0.0817],[784,-0.0924],[785,-0.0955],[787,-0.0985],[788,-0.1013],[789,-0.1046],[790,-0.1046],[792,-0.1114],[793,-0.1114],[794,-0.1147],[795,-0.1147],[796,-0.1181],[797,-0.1181],[798,-0.1214],[801,-0.1214],[802,-0.1248],[803,-0.1248],[834,-0.1339],[837,-0.1341],[887,-0.1147],[888,-0.1181],[889,-0.1181],[890,-0.1214],[893,-0.1214],[894,-0.1248],[895,-0.1248],[926,-0.1339],[929,-0.1341]]
       }
       spine_03_ctrl.Location.Y {
        type=MovieSceneScriptingFloatChannel
        default=0
        thinned_from=735
        keys [[-2,-0.411],[7,-0.414],[9,-0.4201],[10,-0.4202],[12,-0.4265],[13,-0.4266],[14,-0.4298],[15,-0.4309],[20,-0.4324],[21,-0.4308],[22,-0.4313],[24,-0.4287],[26,-0.4292],[27,-0.4277],[28,-0.4277],[29,-0.4266],[30,-0.4266],[33,-0.4236],[40,-0.4189],[45,-0.4156],[48,-0.4156],[49,-0.4123],[54,-0.4119],[55,-0.4086],[61,-0.4078],[62,-0.4045],[63,-0.4053],[65,-0.4035],[66,-0.4033],[68,-0.4014],[70,-0.4012],[84,-0.3995],[85,-0.3969],[86,-0.3968],[87,-0.3941],[88,-0.3889],[89,-0.3862],[94,-0.3598],[95,-0.3571],[96,-0.3517],[97,-0.3479],[98,-0.3424],[99,-0.3387],[100,-0.3331],[101,-0.3292],[102,-0.3238],[106,-0.3088],[110,-0.3014],[111,-0.2985],[112,-0.2986],[113,-0.2967],[115,-0.2969],[118,-0.3033],[119,-0.3075],[120,-0.3136],[121,-0.3176],[122,-0.3257],[123,-0.3318],[125,-0.3477],[126,-0.3576],[127,-0.3652],[132,-0.4093],[137,-0.4312],[141,-0.431],[142,-0.4266],[143,-0.4275],[145,-0.4235],[146,-0.4196],[155,-0.4017],[156,-0.4016],[157,-0.3996],[158,-0.3995],[159,-0.3975],[160,-0.3975],[161,-0.395],[167,-0.3949],[168,-0.3974],[169,-0.3974],[174,-0.4098],[175,-0.4115],[178,-0.4218],[179,-0.4218],[181,-0.4287],[182,-0.4289],[183,-0.4323],[185,-0.4324],[186,-0.4359],[195,-0.4362],[196,-0.4308],[200,-0.4309],[201,-0.4254],[204,-0.4254],[205,-0.42],[206,-0.4201],[207,-0.4192],[208,-0.4164],[209,-0.4164],[210,-0.4135],[211,-0.4135],[213,-0.4077],[214,-0.4077],[216,-0.4019],[217,-0.4019],[219,-0.3961],[222,-0.3789],[223,-0.3694],[224,-0.3576],[225,-0.3516],[226,-0.3397],[227,-0.3336],[228,-0.3217],[232,-0.2974],[233,-0.2973],[234,-0.2912],[235,-0.291],[236,-0.2848],[237,-0.2846],[238,-0.2784],[239,-0.2774],[241,-0.2692],[242,-0.2687],[244,-0.2604],[245,-0.2604],[246,-0.2566],[247,-0.2561],[248,-0.2522],[250,-0.2517],[251,-0.2474],[252,-0.2474],[253,-0.2469],[254,-0.2469],[255,-0.2481],[260,-0.2483],[262,-0.2473],[265,-0.2441],[270,-0.2362],[271,-0.2356],[272,-0.2335],[273,-0.2335],[274,-0.2315],[275,-0.2315],[276,-0.2295],[277,-0.2295],[278,-0.2274],[280,-0.2275],[281,-0.2254],[283,-0.2254],[284,-0.2233],[286,-0.2233],[287,-0.2223],[289,-0.2215],[293,-0.221],[297,-0.2219],[302,-0.2246],[303,-0.2245],[304,-0.2261],[305,-0.226],[306,-0.2277],[307,-0.2276],[308,-0.2294],[309,-0.2292],[323,-0.411],[332,-0.414],[334,-0.4201],[335,-0.4202],[337,-0.4265],[338,-0.4266],[339,-0.4298],[340,-0.4309],[345,-0.4324],[346,-0.4308],[347,-0.4313],[349,-0.4287],[351,-0.4292],[352,-0.4277],[353,-0.4277],[354,-0.4266],[355,-0.4266],[358,-0.4236],[365,-0.4189],[370,-0.4156],[373,-0.4156],[374,-0.4123],[379,-0.4119],[380,-0.4086],[386,-0.4078],[387,-0.4045],[388,-0.4053],[390,-0.4035],[391,-0.4033],[393,-0.4014],[395,-0.4012],[409,-0.3995],[410,-0.3969],[411,-0.3968],[412,-0.3941],[413,-0.3889],[414,-0.3862],[419,-0.3598],[420,-0.3571],[421,-0.3517],[422,-0.3479],[423,-0.3424],[424,-0.3387],[425,-0.3331],[426,-0.3292],[427,-0.3238],[431,-0.3088],[435,-0.3014],[436,-0.2985],[437,-0.2986],[438,-0.2967],[440,-0.2969],[443,-0.3033],[444,-0.3075],[445,-0.3136],[446,-0.3176],[447,-0.3257],[448,-0.3318],[450,-0.3477],[451,-0.3576],[452,-0.3652],[457,-0.4093],[462,-0.4312],[466,-0.431],[467,-0.4266],[468,-0.4275],[470,-0.4235],[471,-0.4196],[480,-0.4017],[481,-0.4016],[482,-0.3996],[483,-0.3995],[484,-0.3975],[485,-0.3975],[486,-0.395],[492,-0.3949],[493,-0.3974],[494,-0.3974],[499,-0.4098],[500,-0.4115],[503,-0.4218],[504,-0.4218],[506,-0.4287],[507,-0.4289],[508,-0.4323],[510,-0.4324],[511,-0.4359],[520,-0.4362],[521,-0.4308],[525,-0.4309],[526,-0.4254],[529,-0.4254],[530,-0.42],[531,-0.4201],[532,-0.4192],[533,-0.4164],[534,-0.4164],[535,-0.4135],[536,-0.4135],[538,-0.4077],[539,-0.4077],[541,-0.4019],[542,-0.4019],[544,-0.3961],[547,-0.3789],[548,-0.3694],[549,-0.3576],[550,-0.3516],[551,-0.3397],[552,-0.3336],[553,-0.3217],[557,-0.2974],[558,-0.2973],[559,-0.2912],[560,-0.291],[561,-0.2848],[562,-0.2846],[563,-0.2784],[564,-0.2774],[566,-0.2692],[567,-0.2687],[569,-0.2604],[570,-0.2604],[571,-0.2566],[572,-0.2561],[573,-0.2522],[575,-0.2517],[576,-0.2474],[577,-0.2474],[578,-0.2469],[579,-0.2469],[580,-0.2481],[585,-0.2483],[587,-0.2473],[590,-0.2441],[595,-0.2362],[596,-0.2356],[597,-0.2335],[598,-0.2335],[599,-0.2315],[600,-0.2315],[601,-0.2295],[602,-0.2295],[603,-0.2274],[605,-0.2275],[606,-0.2254],[608,-0.2254],[609,-0.2233],[611,-0.2233],[612,-0.2223],[614,-0.2215],[618,-0.221],[622,-0.2219],[627,-0.2246],[628,-0.2245],[629,-0.2261],[630,-0.226],[631,-0.2277],[632,-0.2276],[633,-0.2294],[634,-0.2292],[676,-0.411],[685,-0.414],[687,-0.4201],[688,-0.4202],[690,-0.4265],[691,-0.4266],[692,-0.4298],[693,-0.4309],[698,-0.4324],[699,-0.4308],[700,-0.4313],[702,-0.4287],[704,-0.4292],[705,-0.4277],[706,-0.4277],[707,-0.4266],[708,-0.4266],[711,-0.4236],[718,-0.4189],[723,-0.4156],[726,-0.4156],[727,-0.4123],[732,-0.4119],[733,-0.4086],[739,-0.4078],[740,-0.4045],[741,-0.4053],[743,-0.4035],[744,-0.4033],[746,-0.4014],[748,-0.4012],[759,-0.3996],[770,-0.2233],[771,-0.2233],[772,-0.2223],[774,-0.2215],[778,-0.221],[782,-0.2219],[787,-0.2246],[788,-0.2245],[789,-0.2261],[790,-0.226],[791,-0.2277],[792,-0.2276],[793,-0.2294],[794,-0.2292],[795,-0.231],[796,-0.2308],[797,-0.2326],[799,-0.2324],[800,-0.2342],[802,-0.2341],[803,-0.2358],[834,-0.4078],[837,-0.411],[887,-0.231],[888,-0.2308],[889,-0.2326],[891,-0.2324],[892,-0.2342],[894,-0.2341],[895,-0.2358],[926,-0.4078],[929,-0.411]]
       }
       spine_03_ctrl.Location.Z {
        type=MovieSceneScriptingFloatChannel
        default=0
        thinned_from=735
        keys [[-2,-0.0473],[7,-0.0477],[9,-0.0483],[10,-0.0473],[12,-0.046],[13,-0.045],[14,-0.0443],[15,-0.0425],[16,-0.0425],[18,-0.034],[19,-0.034],[20,-0.0297],[21,-0.0295],[22,-0.0252],[23,-0.0251],[24,-0.0206],[25,-0.0206],[26,-0.0163],[28,-0.0162],[29,-0.0117],[31,-0.0115],[33,-0.01],[35,-0.0093],[36,-0.0087],[37,-0.0084],[38,-0.0083],[39,-0.008],[43,-0.0077],[44,-0.0079],[45,-0.0078],[46,-0.008],[47,-0.0075],[48,-0.0075],[49,-0.0072],[50,-0.0072],[51,-0.0109],[54,-0.0109],[55,-0.0106],[56,-0.0106],[57,-0.0143],[60,-0.0143],[61,-0.0181],[62,-0.0177],[63,-0.0178],[64,-0.0195],[65,-0.0195],[67,-0.0231],[68,-0.023],[69,-0.0248],[71,-0.0247],[72,-0.0266],[73,-0.0266],[74,-0.0284],[76,-0.0284],[77,-0.0303],[78,-0.0303],[79,-0.0308],[80,-0.0308],[81,-0.0317],[82,-0.0317],[83,-0.0326],[84,-0.0326],[85,-0.0332],[86,-0.0341],[87,-0.0348],[88,-0.0351],[89,-0.0358],[90,-0.0361],[91,-0.0374],[93,-0.0381],[94,-0.0394],[97,-0.0441],[98,-0.0466],[99,-0.0482],[101,-0.0534],[102,-0.0548],[103,-0.0575],[104,-0.0592],[105,-0.0619],[106,-0.0625],[107,-0.0644],[108,-0.0652],[109,-0.065],[110,-0.0658],[111,-0.0659],[112,-0.065],[113,-0.0648],[115,-0.0629],[118,-0.058],[119,-0.0557],[120,-0.0535],[121,-0.0521],[122,-0.0501],[123,-0.048],[124,-0.0461],[125,-0.0451],[126,-0.0443],[128,-0.0416],[132,-0.0394],[135,-0.0386],[137,-0.0395],[138,-0.0395],[139,-0.0403],[140,-0.0403],[141,-0.041],[142,-0.0413],[143,-0.0422],[145,-0.0434],[146,-0.0438],[154,-0.0488],[155,-0.0486],[156,-0.0494],[157,-0.0501],[158,-0.0509],[159,-0.051],[160,-0.0513],[161,-0.0514],[162,-0.0518],[163,-0.0518],[164,-0.0521],[167,-0.0521],[168,-0.0524],[170,-0.0523],[171,-0.0526],[175,-0.0522],[176,-0.0526],[177,-0.0518],[178,-0.0522],[179,-0.0522],[180,-0.0514],[181,-0.0518],[182,-0.0507],[183,-0.0511],[184,-0.0499],[185,-0.0499],[186,-0.0492],[187,-0.0492],[188,-0.048],[190,-0.048],[191,-0.0473],[192,-0.0473],[193,-0.0461],[195,-0.0461],[196,-0.0455],[197,-0.0455],[198,-0.0442],[200,-0.0442],[201,-0.0436],[204,-0.0436],[205,-0.043],[206,-0.0418],[207,-0.0422],[208,-0.0417],[209,-0.0417],[210,-0.0414],[211,-0.0414],[213,-0.0409],[214,-0.0409],[216,-0.0403],[217,-0.0403],[219,-0.0391],[223,-0.0351],[224,-0.0328],[225,-0.0321],[226,-0.0308],[227,-0.0312],[228,-0.0299],[229,-0.0303],[230,-0.0296],[231,-0.0301],[232,-0.0315],[233,-0.0326],[234,-0.033],[235,-0.0351],[236,-0.0365],[237,-0.0386],[238,-0.0401],[239,-0.0413],[240,-0.0454],[241,-0.0449],[242,-0.0495],[243,-0.049],[244,-0.0531],[245,-0.0531],[246,-0.0527],[247,-0.0572],[248,-0.0568],[249,-0.0613],[250,-0.0613],[251,-0.0654],[252,-0.0654],[253,-0.0699],[254,-0.0699],[258,-0.0766],[260,-0.0793],[261,-0.0804],[262,-0.0812],[264,-0.0824],[266,-0.0827],[271,-0.0818],[272,-0.081],[273,-0.081],[274,-0.0801],[275,-0.0801],[276,-0.0793],[277,-0.0787],[278,-0.0785],[279,-0.0785],[280,-0.0779],[281,-0.0776],[282,-0.0776],[283,-0.0771],[284,-0.0768],[286,-0.0768],[287,-0.0766],[288,-0.0771],[290,-0.0771],[291,-0.0776],[292,-0.0776],[298,-0.0809],[299,-0.082],[301,-0.0832],[302,-0.0844],[303,-0.0844],[305,-0.0872],[306,-0.0874],[307,-0.0887],[308,-0.0889],[309,-0.0902],[323,-0.0473],[332,-0.0477],[334,-0.0483],[335,-0.0473],[337,-0.046],[338,-0.045],[339,-0.0443],[340,-0.0425],[341,-0.0425],[343,-0.034],[344,-0.034],[345,-0.0297],[346,-0.0295],[347,-0.0252],[348,-0.0251],[349,-0.0206],[350,-0.0206],[351,-0.0163],[353,-0.0162],[354,-0.0117],[356,-0.0115],[358,-0.01],[360,-0.0093],[361,-0.0087],[362,-0.0084],[363,-0.0083],[364,-0.008],[368,-0.0077],[369,-0.0079],[370,-0.0078],[371,-0.008],[372,-0.0075],[373,-0.0075],[374,-0.0072],[375,-0.0072],[376,-0.0109],[379,-0.0109],[380,-0.0106],[381,-0.0106],[382,-0.0143],[385,-0.0143],[386,-0.0181],[387,-0.0177],[388,-0.0178],[389,-0.0195],[390,-0.0195],[392,-0.0231],[393,-0.023],[394,-0.0248],[396,-0.0247],[397,-0.0266],[398,-0.0266],[399,-0.0284],[401,-0.0284],[402,-0.0303],[403,-0.0303],[404,-0.0308],[405,-0.0308],[406,-0.0317],[407,-0.0317],[408,-0.0326],[409,-0.0326],[410,-0.0332],[411,-0.0341],[412,-0.0348],[413,-0.0351],[414,-0.0358],[415,-0.0361],[416,-0.0374],[418,-0.0381],[419,-0.0394],[422,-0.0441],[423,-0.0466],[424,-0.0482],[426,-0.0534],[427,-0.0548],[428,-0.0575],[429,-0.0592],[430,-0.0619],[431,-0.0625],[432,-0.0644],[433,-0.0652],[434,-0.065],[435,-0.0658],[436,-0.0659],[437,-0.065],[438,-0.0648],[440,-0.0629],[443,-0.058],[444,-0.0557],[445,-0.0535],[446,-0.0521],[447,-0.0501],[448,-0.048],[449,-0.0461],[450,-0.0451],[451,-0.0443],[453,-0.0416],[457,-0.0394],[460,-0.0386],[462,-0.0395],[463,-0.0395],[464,-0.0403],[465,-0.0403],[466,-0.041],[467,-0.0413],[468,-0.0422],[470,-0.0434],[471,-0.0438],[479,-0.0488],[480,-0.0486],[481,-0.0494],[482,-0.0501],[483,-0.0509],[484,-0.051],[485,-0.0513],[486,-0.0514],[487,-0.0518],[488,-0.0518],[489,-0.0521],[492,-0.0521],[493,-0.0524],[495,-0.0523],[496,-0.0526],[500,-0.0522],[501,-0.0526],[502,-0.0518],[503,-0.0522],[504,-0.0522],[505,-0.0514],[506,-0.0518],[507,-0.0507],[508,-0.0511],[509,-0.0499],[510,-0.0499],[511,-0.0492],[512,-0.0492],[513,-0.048],[515,-0.048],[516,-0.0473],[517,-0.0473],[518,-0.0461],[520,-0.0461],[521,-0.0455],[522,-0.0455],[523,-0.0442],[525,-0.0442],[526,-0.0436],[529,-0.0436],[530,-0.043],[531,-0.0418],[532,-0.0422],[533,-0.0417],[534,-0.0417],[535,-0.0414],[536,-0.0414],[538,-0.0409],[539,-0.0409],[541,-0.0403],[542,-0.0403],[544,-0.0391],[548,-0.0351],[549,-0.0328],[550,-0.0321],[551,-0.0308],[552,-0.0312],[553,-0.0299],[554,-0.0303],[555,-0.0296],[556,-0.0301],[557,-0.0315],[558,-0.0326],[559,-0.033],[560,-0.0351],[561,-0.0365],[562,-0.0386],[563,-0.0401],[564,-0.0413],[565,-0.0454],[566,-0.0449],[567,-0.0495],[568,-0.049],[569,-0.0531],[570,-0.0531],[571,-0.0527],[572,-0.0572],[573,-0.0568],[574,-0.0613],[575,-0.0613],[576,-0.0654],[577,-0.0654],[578,-0.0699],[579,-0.0699],[583,-0.0766],[585,-0.0793],[586,-0.0804],[587,-0.0812],[589,-0.0824],[591,-0.0827],[596,-0.0818],[597,-0.081],[598,-0.081],[599,-0.0801],[600,-0.0801],[601,-0.0793],[602,-0.0787],[603,-0.0785],[604,-0.0785],[605,-0.0779],[606,-0.0776],[607,-0.0776],[608,-0.0771],[609,-0.0768],[611,-0.0768],[612,-0.0766],[613,-0.0771],[615,-0.0771],[616,-0.0776],[617,-0.0776],[623,-0.0809],[624,-0.082],[626,-0.0832],[627,-0.0844],[628,-0.0844],[630,-0.0872],[631,-0.0874],[632,-0.0887],[633,-0.0889],[634,-0.0902],[676,-0.0473],[685,-0.0477],[687,-0.0483],[688,-0.0473],[690,-0.046],[691,-0.045],[692,-0.0443],[693,-0.0425],[694,-0.0425],[696,-0.034],[697,-0.034],[698,-0.0297],[699,-0.0295],[700,-0.0252],[701,-0.0251],[702,-0.0206],[703,-0.0206],[704,-0.0163],[706,-0.0162],[707,-0.0117],[709,-0.0115],[711,-0.01],[713,-0.0093],[714,-0.0087],[715,-0.0084],[716,-0.0083],[717,-0.008],[721,-0.0077],[722,-0.0079],[723,-0.0078],[724,-0.008],[725,-0.0075],[726,-0.0075],[727,-0.0072],[728,-0.0072],[729,-0.0109],[732,-0.0109],[733,-0.0106],[734,-0.0106],[735,-0.0143],[738,-0.0143],[739,-0.0181],[740,-0.0177],[741,-0.0178],[742,-0.0195],[743,-0.0195],[745,-0.0231],[746,-0.023],[747,-0.0248],[749,-0.0247],[750,-0.0266],[751,-0.0266],[752,-0.0284],[754,-0.0284],[755,-0.0303],[757,-0.0308],[759,-0.0317],[770,-0.0768],[772,-0.0766],[773,-0.0771],[775,-0.0771],[776,-0.0776],[777,-0.0776],[783,-0.0809],[784,-0.082],[786,-0.0832],[787,-0.0844],[788,-0.0844],[790,-0.0872],[791,-0.0874],[792,-0.0887],[793,-0.0889],[794,-0.0902],[795,-0.0904],[796,-0.0917],[798,-0.0919],[799,-0.0932],[801,-0.0934],[802,-0.0947],[803,-0.0949],[834,-0.0484],[837,-0.0473],[887,-0.0904],[888,-0.0917],[890,-0.0919],[891,-0.0932],[893,-0.0934],[894,-0.0947],[895,-0.0949],[926,-0.0484],[929,-0.0473]]
       }
       spine_03_ctrl.Rotation.X {
        type=MovieSceneScriptingFloatChannel
        default=0
        thinned_from=717
        keys [[-2,-2.6387],[-1,-2.6851],[10,-3.1083],[16,-3.1633],[56,-3.0568],[85,-2.8926],[87,-2.863],[99,-2.5797],[106,-2.3849],[110,-2.2943],[112,-2.2568],[113,-2.2201],[114,-2.1908],[115,-2.1727],[116,-2.1662],[117,-2.1717],[118,-2.1891],[119,-2.2199],[120,-2.2641],[121,-2.32],[122,-2.3862],[123,-2.4647],[124,-2.554],[125,-2.6522],[129,-3.085],[130,-3.1875],[131,-3.2831],[132,-3.3718],[133,-3.4491],[134,-3.5148],[135,-3.5661],[136,-3.602],[137,-3.6212],[138,-3.6273],[140,-3.6167],[142,-3.5801],[144,-3.524],[150,-3.3257],[153,-3.2445],[159,-3.1285],[162,-3.086],[164,-3.0735],[167,-3.0811],[169,-3.1013],[172,-3.1568],[174,-3.2048],[182,-3.4505],[186,-3.5316],[190,-3.5792],[194,-3.5908],[201,-3.5508],[207,-3.478],[212,-3.38],[217,-3.2534],[218,-3.2194],[219,-3.1761],[220,-3.1201],[221,-3.0386],[222,-2.9378],[223,-2.8234],[226,-2.448],[227,-2.3307],[228,-2.2237],[229,-2.1282],[230,-2.0467],[231,-1.9836],[232,-1.9341],[233,-1.8933],[235,-1.8277],[236,-1.7992],[239,-1.7323],[242,-1.6736],[244,-1.6463],[246,-1.6298],[247,-1.6255],[248,-1.6257],[249,-1.6298],[250,-1.6394],[251,-1.6533],[253,-1.6938],[255,-1.7486],[257,-1.8145],[261,-1.9595],[263,-2.021],[265,-2.0717],[270,-2.1733],[273,-2.2497],[287,-2.643],[302,-2.9672],[308,-3.0586],[309,-3.0702],[335,-3.1083],[341,-3.1633],[381,-3.0568],[410,-2.8926],[412,-2.863],[424,-2.5797],[431,-2.3849],[435,-2.2943],[437,-2.2568],[438,-2.2201],[439,-2.1908],[440,-2.1727],[441,-2.1662],[442,-2.1717],[443,-2.1891],[444,-2.2199],[445,-2.2641],[446,-2.32],[447,-2.3862],[448,-2.4647],[449,-2.554],[450,-2.6522],[454,-3.085],[455,-3.1875],[456,-3.2831],[457,-3.3718],[458,-3.4491],[459,-3.5148],[460,-3.5661],[461,-3.602],[462,-3.6212],[463,-3.6273],[465,-3.6167],[467,-3.5801],[469,-3.524],[475,-3.3257],[478,-3.2445],[484,-3.1285],[487,-3.086],[489,-3.0735],[492,-3.0811],[494,-3.1013],[497,-3.1568],[499,-3.2048],[507,-3.4505],[511,-3.5316],[515,-3.5792],[519,-3.5908],[526,-3.5508],[532,-3.478],[537,-3.38],[542,-3.2534],[543,-3.2194],[544,-3.1761],[545,-3.1201],[546,-3.0386],[547,-2.9378],[548,-2.8234],[551,-2.448],[552,-2.3307],[553,-2.2237],[554,-2.1282],[555,-2.0467],[556,-1.9836],[557,-1.9341],[558,-1.8933],[560,-1.8277],[561,-1.7992],[564,-1.7323],[567,-1.6736],[569,-1.6463],[571,-1.6298],[572,-1.6255],[573,-1.6257],[574,-1.6298],[575,-1.6394],[576,-1.6533],[578,-1.6938],[580,-1.7486],[582,-1.8145],[586,-1.9595],[588,-2.021],[590,-2.0717],[595,-2.1733],[598,-2.2497],[612,-2.643],[627,-2.9672],[633,-3.0586],[634,-3.0702],[676,14.9256],[688,14.3449],[759,14.5845],[774,14.8611],[796,14.3697],[800,14.3914],[803,14.4838],[834,15.0216],[851,14.7557],[887,-3.0803],[889,-3.0938],[890,-3.0926],[892,-3.0716],[894,-3.0272],[895,-2.9969],[926,-2.5623],[929,-2.6387]]
       }
       spine_03_ctrl.Rotation.Y {
        type=MovieSceneScriptingFloatChannel
        default=0
        thinned_from=720
        keys [[-2,1.7435],[-1,1.7439],[11,1.7812],[13,1.7997],[18,1.8601],[21,1.8865],[23,1.8953],[49,1.9105],[50,1.9088],[64,1.896],[82,1.8388],[86,1.8054],[90,1.7548],[94,1.683],[96,1.6392],[99,1.5589],[101,1.4973],[103,1.4271],[104,1.3886],[105,1.3471],[107,1.2556],[111,1.0539],[117,0.7435],[121,0.5297],[122,0.477],[123,0.4256],[124,0.3755],[125,0.3278],[126,0.2826],[127,0.2406],[128,0.2029],[129,0.1686],[130,0.1374],[131,0.1089],[132,0.0838],[133,0.0624],[134,0.0447],[135,0.0306],[136,0.0217],[137,0.0171],[138,0.0166],[139,0.0182],[140,0.0226],[141,0.0283],[142,0.0359],[143,0.045],[144,0.056],[145,0.0674],[146,0.0809],[147,0.0925],[148,0.1053],[149,0.1174],[150,0.1289],[151,0.1409],[152,0.1518],[153,0.1632],[155,0.1831],[156,0.1924],[159,0.2179],[161,0.233],[162,0.2391],[163,0.2437],[164,0.2469],[165,0.2485],[167,0.2489],[168,0.2476],[169,0.2453],[171,0.2383],[172,0.2356],[174,0.2278],[176,0.2226],[178,0.2198],[180,0.219],[181,0.218],[182,0.219],[184,0.2197],[186,0.2223],[188,0.226],[191,0.2347],[194,0.2452],[196,0.2514],[197,0.2551],[200,0.2634],[201,0.2658],[203,0.2689],[205,0.2705],[208,0.2692],[210,0.267],[211,0.2654],[212,0.2626],[213,0.2604],[214,0.2635],[215,0.2756],[216,0.2965],[217,0.3247],[218,0.3584],[219,0.395],[220,0.434],[222,0.5178],[228,0.7762],[230,0.8658],[232,0.9573],[234,1.0432],[235,1.0825],[236,1.1165],[237,1.1418],[238,1.1591],[239,1.1677],[240,1.1695],[241,1.1647],[242,1.152],[243,1.1307],[309,1.5843],[323,1.7435],[324,1.7439],[336,1.7812],[338,1.7997],[343,1.8601],[346,1.8865],[348,1.8953],[374,1.9105],[375,1.9088],[389,1.896],[407,1.8388],[411,1.8054],[415,1.7548],[419,1.683],[421,1.6392],[424,1.5589],[426,1.4973],[428,1.4271],[429,1.3886],[430,1.3471],[432,1.2556],[436,1.0539],[442,0.7435],[446,0.5297],[447,0.477],[448,0.4256],[449,0.3755],[450,0.3278],[451,0.2826],[452,0.2406],[453,0.2029],[454,0.1686],[455,0.1374],[456,0.1089],[457,0.0838],[458,0.0624],[459,0.0447],[460,0.0306],[461,0.0217],[462,0.0171],[463,0.0166],[464,0.0182],[465,0.0226],[466,0.0283],[467,0.0359],[468,0.045],[469,0.056],[470,0.0674],[471,0.0809],[472,0.0925],[473,0.1053],[474,0.1174],[475,0.1289],[476,0.1409],[477,0.1518],[478,0.1632],[480,0.1831],[481,0.1924],[484,0.2179],[486,0.233],[487,0.2391],[488,0.2437],[489,0.2469],[490,0.2485],[492,0.2489],[493,0.2476],[494,0.2453],[496,0.2383],[497,0.2356],[499,0.2278],[501,0.2226],[503,0.2198],[505,0.219],[506,0.218],[507,0.219],[509,0.2197],[511,0.2223],[513,0.226],[516,0.2347],[519,0.2452],[521,0.2514],[522,0.2551],[525,0.2634],[526,0.2658],[528,0.2689],[530,0.2705],[533,0.2692],[535,0.267],[536,0.2654],[537,0.2626],[538,0.2604],[539,0.2635],[540,0.2756],[541,0.2965],[542,0.3247],[543,0.3584],[544,0.395],[545,0.434],[547,0.5178],[553,0.7762],[555,0.8658],[557,0.9573],[559,1.0432],[560,1.0825],[561,1.1165],[562,1.1418],[563,1.1591],[564,1.1677],[565,1.1695],[566,1.1647],[567,1.152],[568,1.1307],[569,1.1016],[570,1.0641],[571,1.0195],[572,0.9685],[573,0.9131],[574,0.8518],[575,0.7883],[576,0.7222],[577,0.6544],[578,0.5843],[579,0.5124],[580,0.4395],[581,0.364],[582,0.2857],[583,0.2047],[584,0.1213],[585,0.0342],[586,-0.0566],[587,-0.1528],[588,-0.2529],[589,-0.3575],[590,-0.4657],[591,-0.5775],[592,-0.6924],[593,-0.8092],[597,-1.2912],[604,-2.1476],[607,-2.4976],[610,-2.8283],[613,-3.1345],[615,-3.3237],[618,-3.5793],[620,-3.7306],[622,-3.8656],[624,-3.9852],[626,-4.0856],[628,-4.1676],[630,-4.2305],[632,-4.273],[634,-4.2953],[676,-1.0447],[677,-1.0443],[689,-1.007],[690,-0.999],[696,-0.9281],[698,-0.9092],[699,-0.9017],[701,-0.8929],[712,-0.8952],[727,-0.8777],[728,-0.8794],[729,-0.876],[735,-0.8786],[742,-0.8922],[757,-0.9403],[759,-0.9452],[770,-1.002],[799,-1.0475],[803,-1.0481],[834,-1.0362],[837,-1.0447],[887,-4.2986],[888,-4.2965],[889,-4.279],[890,-4.2378],[891,-4.1739],[892,-4.0894],[893,-3.9844],[894,-3.8599],[895,-3.7167],[926,1.752],[929,1.7435]]
       }
       spine_03_ctrl.Rotation.Z {
        type=MovieSceneScriptingFloatChannel
        default=0
        thinned_from=620
        keys [[-2,0.2767],[-1,0.2664],[6,-0.0865],[8,-0.2338],[9,-0.3123],[11,-0.4798],[14,-0.746],[16,-0.9286],[19,-1.196],[20,-1.2819],[22,-1.4446],[24,-1.5918],[25,-1.6578],[26,-1.7135],[27,-1.761],[28,-1.8001],[30,-1.8556],[34,-1.9134],[42,-1.9927],[49,-2.0309],[56,-2.0325],[71,-1.9658],[83,-1.9859],[84,-2.0388],[85,-2.1221],[86,-2.2304],[87,-2.3599],[88,-2.5127],[89,-2.694],[90,-2.8914],[93,-3.5146],[94,-3.7041],[95,-3.8733],[96,-4.0091],[97,-4.1084],[98,-4.1763],[99,-4.2159],[100,-4.2271],[101,-4.2109],[102,-4.1674],[103,-4.0956],[104,-3.9964],[105,-3.8758],[106,-3.745],[111,-3.0238],[114,-2.6291],[115,-2.5021],[116,-2.3817],[117,-2.2748],[118,-2.1877],[119,-2.1235],[120,-2.0861],[121,-2.071],[122,-2.0707],[123,-2.0892],[124,-2.1268],[125,-2.1814],[126,-2.2487],[129,-2.4787],[130,-2.5499],[131,-2.6136],[132,-2.6701],[133,-2.7185],[134,-2.7609],[136,-2.8233],[137,-2.8473],[138,-2.8646],[152,-2.8724],[168,-2.818],[171,-2.8184],[174,-2.8513],[175,-2.872],[177,-2.9402],[179,-3.0201],[181,-3.0866],[182,-3.113],[183,-3.132],[184,-3.1429],[189,-3.1518],[191,-3.1458],[193,-3.1301],[194,-3.1173],[204,-2.6271],[206,-2.5013],[207,-2.4294],[208,-2.3439],[209,-2.2463],[210,-2.1352],[211,-2.0116],[212,-1.8751],[213,-1.7168],[214,-1.5343],[215,-1.3198],[216,-1.0772],[217,-0.8159],[218,-0.5407],[219,-0.2563],[220,0.0342],[221,0.3231],[222,0.6076],[223,0.8829],[224,1.1445],[225,1.3864],[226,1.6041],[227,1.7948],[228,1.951],[229,2.0818],[230,2.1919],[231,2.2857],[232,2.3616],[235,2.4555],[236,2.4721],[243,2.5166],[244,2.5008],[245,2.4759],[246,2.4427],[247,2.4013],[248,2.3534],[250,2.2396],[253,2.0385],[256,1.8238],[258,1.6931],[259,1.6358],[260,1.5825],[261,1.5367],[262,1.4951],[263,1.4638],[264,1.4397],[265,1.425],[266,1.4197],[267,1.4272],[269,1.4503],[272,1.4728],[276,1.4885],[290,1.4993],[292,1.4877],[294,1.4632],[295,1.4471],[296,1.4231],[297,1.3938],[299,1.3117],[300,1.2614],[301,1.2062],[303,1.0837],[304,1.0181],[306,0.8777],[309,0.6607],[323,0.2767],[324,0.2664],[331,-0.0865],[333,-0.2338],[334,-0.3123],[336,-0.4798],[339,-0.746],[341,-0.9286],[344,-1.196],[345,-1.2819],[347,-1.4446],[349,-1.5918],[350,-1.6578],[351,-1.7135],[352,-1.761],[353,-1.8001],[355,-1.8556],[359,-1.9134],[367,-1.9927],[374,-2.0309],[381,-2.0325],[396,-1.9658],[408,-1.9859],[409,-2.0388],[410,-2.1221],[411,-2.2304],[412,-2.3599],[413,-2.5127],[414,-2.694],[415,-2.8914],[418,-3.5146],[419,-3.7041],[420,-3.8733],[421,-4.0091],[422,-4.1084],[423,-4.1763],[424,-4.2159],[425,-4.2271],[426,-4.2109],[427,-4.1674],[428,-4.0956],[429,-3.9964],[430,-3.8758],[431,-3.745],[436,-3.0238],[439,-2.6291],[440,-2.5021],[441,-2.3817],[442,-2.2748],[443,-2.1877],[444,-2.1235],[445,-2.0861],[446,-2.071],[447,-2.0707],[448,-2.0892],[449,-2.1268],[450,-2.1814],[451,-2.2487],[454,-2.4787],[455,-2.5499],[456,-2.6136],[457,-2.6701],[458,-2.7185],[459,-2.7609],[461,-2.8233],[462,-2.8473],[463,-2.8646],[477,-2.8724],[493,-2.818],[496,-2.8184],[499,-2.8513],[500,-2.872],[502,-2.9402],[504,-3.0201],[506,-3.0866],[507,-3.113],[508,-3.132],[509,-3.1429],[514,-3.1518],[516,-3.1458],[518,-3.1301],[519,-3.1173],[529,-2.6271],[531,-2.5013],[532,-2.4294],[533,-2.3439],[534,-2.2463],[535,-2.1352],[536,-2.0116],[537,-1.8751],[538,-1.7168],[539,-1.5343],[540,-1.3198],[541,-1.0772],[542,-0.8159],[543,-0.5407],[544,-0.2563],[545,0.0342],[546,0.3231],[547,0.6076],[548,0.8829],[549,1.1445],[550,1.3864],[551,1.6041],[552,1.7948],[553,1.951],[554,2.0818],[555,2.1919],[556,2.2857],[557,2.3616],[560,2.4555],[561,2.4721],[568,2.5166],[569,2.5008],[570,2.4759],[571,2.4427],[572,2.4013],[573,2.3534],[575,2.2396],[578,2.0385],[581,1.8238],[583,1.6931],[584,1.6358],[585,1.5825],[586,1.5367],[587,1.4951],[588,1.4638],[589,1.4397],[590,1.425],[591,1.4197],[592,1.4272],[594,1.4503],[597,1.4728],[601,1.4885],[615,1.4993],[617,1.4877],[619,1.4632],[620,1.4471],[621,1.4231],[622,1.3938],[624,1.3117],[625,1.2614],[626,1.2062],[628,1.0837],[629,1.0181],[631,0.8777],[634,0.6607],[676,0.4052],[677,0.3949],[684,0.042],[685,-0.0315],[686,-0.1053],[687,-0.1839],[688,-0.2673],[689,-0.3514],[692,-0.6175],[694,-0.8002],[696,-0.9773],[697,-1.0675],[698,-1.1534],[700,-1.3161],[702,-1.4633],[703,-1.5293],[704,-1.5851],[705,-1.6325],[706,-1.6716],[708,-1.7271],[710,-1.7587],[712,-1.7849],[720,-1.8643],[727,-1.9024],[732,-1.9069],[739,-1.8868],[753,-1.8234],[757,-1.8148],[759,-1.814],[803,0.252],[834,0.4066],[837,0.4052],[887,0.5891],[888,0.518],[889,0.4518],[890,0.3868],[891,0.3252],[892,0.2688],[893,0.2154],[894,0.1671],[895,0.1235],[926,0.2782],[929,0.2767]]
       }
       spine_03_ctrl.Scale.X {type=MovieSceneScriptingFloatChannel, default=1, const [1,735]}
       spine_03_ctrl.Scale.Y {type=MovieSceneScriptingFloatChannel, default=1, const [1,735]}
       spine_03_ctrl.Scale.Z {type=MovieSceneScriptingFloatChannel, default=1, const [1,735]}
       spine_04_ctrl.Location.X {
        type=MovieSceneScriptingFloatChannel
        default=0
        thinned_from=735
        keys [[-2,-0.0401],[6,-0.0413],[9,-0.0413],[10,-0.0424],[13,-0.0424],[14,-0.0436],[15,-0.0437],[16,-0.0442],[17,-0.0444],[18,-0.0449],[20,-0.0454],[21,-0.0459],[23,-0.0464],[24,-0.0464],[26,-0.0468],[27,-0.0468],[28,-0.0471],[38,-0.047],[39,-0.0484],[84,-0.0461],[88,-0.0464],[93,-0.0474],[98,-0.0476],[99,-0.0454],[101,-0.0454],[102,-0.0431],[103,-0.0431],[104,-0.0408],[105,-0.0408],[107,-0.0362],[108,-0.0362],[110,-0.0317],[111,-0.0303],[112,-0.0275],[113,-0.0275],[114,-0.0247],[115,-0.0247],[117,-0.0192],[118,-0.0192],[119,-0.0164],[121,-0.0164],[122,-0.0136],[124,-0.0136],[125,-0.0109],[126,-0.0109],[127,-0.0106],[128,-0.0099],[129,-0.0099],[130,-0.0092],[133,-0.0092],[134,-0.0085],[139,-0.0085],[140,-0.0092],[145,-0.0093],[146,-0.0103],[150,-0.0103],[151,-0.0114],[156,-0.0114],[157,-0.0124],[174,-0.0123],[175,-0.0132],[177,-0.0132],[178,-0.0142],[181,-0.0142],[182,-0.0151],[188,-0.0151],[189,-0.016],[190,-0.016],[191,-0.0163],[202,-0.0163],[203,-0.0151],[207,-0.0151],[208,-0.0142],[210,-0.0142],[211,-0.0132],[213,-0.0132],[214,-0.0123],[216,-0.0123],[217,-0.0132],[219,-0.0132],[220,-0.0142],[222,-0.0142],[224,-0.0143],[225,-0.0149],[229,-0.0149],[230,-0.0155],[231,-0.0155],[232,-0.0161],[233,-0.0161],[235,-0.0172],[236,-0.0172],[238,-0.0184],[239,-0.0184],[245,-0.0201],[247,-0.0201],[249,-0.0207],[250,-0.0207],[251,-0.021],[252,-0.0216],[253,-0.0219],[258,-0.0253],[259,-0.0253],[261,-0.0268],[262,-0.0268],[263,-0.0275],[266,-0.0275],[267,-0.0282],[270,-0.0282],[271,-0.028],[286,-0.028],[287,-0.0292],[288,-0.0292],[289,-0.03],[291,-0.03],[292,-0.0308],[294,-0.0308],[295,-0.0317],[296,-0.0317],[297,-0.0325],[298,-0.0325],[300,-0.0341],[301,-0.0341],[302,-0.035],[304,-0.0351],[305,-0.0369],[308,-0.0369],[309,-0.0387],[324,-0.0402],[331,-0.0413],[334,-0.0413],[335,-0.0424],[338,-0.0424],[339,-0.0436],[340,-0.0437],[341,-0.0442],[342,-0.0444],[343,-0.0449],[345,-0.0454],[346,-0.0459],[348,-0.0464],[349,-0.0464],[351,-0.0468],[352,-0.0468],[353,-0.0471],[363,-0.047],[364,-0.0484],[409,-0.0461],[413,-0.0464],[418,-0.0474],[423,-0.0476],[424,-0.0454],[426,-0.0454],[427,-0.0431],[428,-0.0431],[429,-0.0408],[430,-0.0408],[432,-0.0362],[433,-0.0362],[435,-0.0317],[436,-0.0303],[437,-0.0275],[438,-0.0275],[439,-0.0247],[440,-0.0247],[442,-0.0192],[443,-0.0192],[444,-0.0164],[446,-0.0164],[447,-0.0136],[449,-0.0136],[450,-0.0109],[451,-0.0109],[452,-0.0106],[453,-0.0099],[454,-0.0099],[455,-0.0092],[458,-0.0092],[459,-0.0085],[464,-0.0085],[465,-0.0092],[470,-0.0093],[471,-0.0103],[475,-0.0103],[476,-0.0114],[481,-0.0114],[482,-0.0124],[499,-0.0123],[500,-0.0132],[502,-0.0132],[503,-0.0142],[506,-0.0142],[507,-0.0151],[513,-0.0151],[514,-0.016],[515,-0.016],[516,-0.0163],[527,-0.0163],[528,-0.0151],[532,-0.0151],[533,-0.0142],[535,-0.0142],[536,-0.0132],[538,-0.0132],[539,-0.0123],[541,-0.0123],[542,-0.0132],[544,-0.0132],[545,-0.0142],[547,-0.0142],[549,-0.0143],[550,-0.0149],[554,-0.0149],[555,-0.0155],[556,-0.0155],[557,-0.0161],[558,-0.0161],[560,-0.0172],[561,-0.0172],[563,-0.0184],[564,-0.0184],[570,-0.0201],[572,-0.0201],[574,-0.0207],[575,-0.0207],[576,-0.021],[577,-0.0216],[578,-0.0219],[583,-0.0253],[584,-0.0253],[586,-0.0268],[587,-0.0268],[588,-0.0275],[591,-0.0275],[592,-0.0282],[595,-0.0282],[596,-0.028],[611,-0.028],[612,-0.0292],[613,-0.0292],[614,-0.03],[616,-0.03],[617,-0.0308],[619,-0.0308],[620,-0.0317],[621,-0.0317],[622,-0.0325],[623,-0.0325],[625,-0.0341],[626,-0.0341],[627,-0.035],[629,-0.0351],[630,-0.0369],[633,-0.0369],[634,-0.0387],[677,-0.0402],[684,-0.0413],[687,-0.0413],[688,-0.0424],[691,-0.0424],[692,-0.0436],[693,-0.0437],[694,-0.0442],[695,-0.0444],[696,-0.0449],[698,-0.0454],[699,-0.0459],[701,-0.0464],[702,-0.0464],[704,-0.0468],[705,-0.0468],[706,-0.0471],[716,-0.047],[717,-0.0484],[759,-0.0463],[770,-0.028],[771,-0.028],[772,-0.0292],[773,-0.0292],[774,-0.03],[776,-0.03],[777,-0.0308],[779,-0.0308],[780,-0.0317],[781,-0.0317],[782,-0.0325],[783,-0.0325],[785,-0.0341],[786,-0.0341],[787,-0.035],[789,-0.0351],[790,-0.0369],[793,-0.0369],[794,-0.0387],[798,-0.0387],[799,-0.0404],[837,-0.0401],[890,-0.0387],[891,-0.0404],[929,-0.0401]]
       }
       spine_04_ctrl.Location.Y {
        type=MovieSceneScriptingFloatChannel
        default=0
        thinned_from=735
        keys [[-2,-0.1025],[-1,-0.1027],[6,-0.1052],[8,-0.1052],[9,-0.1082],[11,-0.1082],[12,-0.1112],[13,-0.1112],[14,-0.1115],[16,-0.1125],[19,-0.1136],[28,-0.1139],[29,-0.1136],[33,-0.1137],[34,-0.114],[38,-0.114],[39,-0.1118],[46,-0.1118],[47,-0.112],[48,-0.1115],[50,-0.1114],[51,-0.111],[53,-0.1108],[54,-0.1104],[56,-0.1102],[57,-0.1096],[58,-0.1096],[59,-0.109],[60,-0.1089],[61,-0.1084],[62,-0.1083],[72,-0.1051],[75,-0.1047],[76,-0.1045],[83,-0.104],[84,-0.1032],[85,-0.1031],[87,-0.1015],[88,-0.1005],[89,-0.0988],[90,-0.098],[94,-0.0912],[95,-0.0904],[96,-0.0898],[97,-0.087],[98,-0.0865],[99,-0.0836],[100,-0.0831],[101,-0.0803],[102,-0.0797],[103,-0.0764],[104,-0.0735],[105,-0.073],[106,-0.073],[107,-0.0702],[108,-0.0696],[109,-0.0668],[112,-0.0666],[113,-0.0643],[115,-0.0643],[116,-0.0648],[117,-0.0648],[118,-0.0675],[119,-0.0675],[120,-0.0703],[121,-0.0707],[122,-0.073],[124,-0.0785],[125,-0.0808],[127,-0.0861],[128,-0.0867],[129,-0.0936],[130,-0.0936],[131,-0.1011],[133,-0.1011],[134,-0.1086],[141,-0.1086],[142,-0.1081],[143,-0.1066],[144,-0.1066],[145,-0.1061],[146,-0.1026],[149,-0.1026],[150,-0.0985],[154,-0.0985],[155,-0.0944],[174,-0.0943],[175,-0.0976],[176,-0.0976],[177,-0.1006],[178,-0.1009],[180,-0.1009],[181,-0.1038],[182,-0.1038],[183,-0.1041],[185,-0.1041],[186,-0.1071],[188,-0.1071],[189,-0.1074],[196,-0.1074],[197,-0.1077],[201,-0.1077],[202,-0.1036],[211,-0.1034],[212,-0.0983],[218,-0.0982],[219,-0.0931],[221,-0.0932],[222,-0.088],[223,-0.0856],[224,-0.0818],[225,-0.0778],[226,-0.0778],[228,-0.0703],[229,-0.07],[230,-0.0662],[231,-0.0662],[232,-0.0622],[233,-0.0619],[234,-0.0619],[235,-0.0616],[236,-0.0578],[239,-0.0569],[242,-0.0539],[243,-0.0533],[244,-0.0524],[249,-0.0494],[254,-0.0482],[255,-0.0486],[256,-0.0484],[257,-0.0486],[258,-0.0484],[260,-0.0486],[262,-0.0482],[264,-0.0482],[265,-0.0476],[266,-0.0476],[270,-0.0461],[271,-0.0459],[286,-0.0459],[287,-0.0434],[291,-0.0432],[294,-0.0429],[299,-0.0431],[302,-0.0434],[304,-0.0432],[305,-0.0443],[308,-0.0443],[309,-0.044],[323,-0.1025],[324,-0.1027],[331,-0.1052],[333,-0.1052],[334,-0.1082],[336,-0.1082],[337,-0.1112],[338,-0.1112],[339,-0.1115],[341,-0.1125],[344,-0.1136],[353,-0.1139],[354,-0.1136],[358,-0.1137],[359,-0.114],[363,-0.114],[364,-0.1118],[371,-0.1118],[372,-0.112],[373,-0.1115],[375,-0.1114],[376,-0.111],[378,-0.1108],[379,-0.1104],[381,-0.1102],[382,-0.1096],[383,-0.1096],[384,-0.109],[385,-0.1089],[386,-0.1084],[387,-0.1083],[397,-0.1051],[400,-0.1047],[401,-0.1045],[408,-0.104],[409,-0.1032],[410,-0.1031],[412,-0.1015],[413,-0.1005],[414,-0.0988],[415,-0.098],[419,-0.0912],[420,-0.0904],[421,-0.0898],[422,-0.087],[423,-0.0865],[424,-0.0836],[425,-0.0831],[426,-0.0803],[427,-0.0797],[428,-0.0764],[429,-0.0735],[430,-0.073],[431,-0.073],[432,-0.0702],[433,-0.0696],[434,-0.0668],[437,-0.0666],[438,-0.0643],[440,-0.0643],[441,-0.0648],[442,-0.0648],[443,-0.0675],[444,-0.0675],[445,-0.0703],[446,-0.0707],[447,-0.073],[449,-0.0785],[450,-0.0808],[452,-0.0861],[453,-0.0867],[454,-0.0936],[455,-0.0936],[456,-0.1011],[458,-0.1011],[459,-0.1086],[466,-0.1086],[467,-0.1081],[468,-0.1066],[469,-0.1066],[470,-0.1061],[471,-0.1026],[474,-0.1026],[475,-0.0985],[479,-0.0985],[480,-0.0944],[499,-0.0943],[500,-0.0976],[501,-0.0976],[502,-0.1006],[503,-0.1009],[505,-0.1009],[506,-0.1038],[507,-0.1038],[508,-0.1041],[510,-0.1041],[511,-0.1071],[513,-0.1071],[514,-0.1074],[521,-0.1074],[522,-0.1077],[526,-0.1077],[527,-0.1036],[536,-0.1034],[537,-0.0983],[543,-0.0982],[544,-0.0931],[546,-0.0932],[547,-0.088],[548,-0.0856],[549,-0.0818],[550,-0.0778],[551,-0.0778],[553,-0.0703],[554,-0.07],[555,-0.0662],[556,-0.0662],[557,-0.0622],[558,-0.0619],[559,-0.0619],[560,-0.0616],[561,-0.0578],[564,-0.0569],[567,-0.0539],[568,-0.0533],[569,-0.0524],[574,-0.0494],[579,-0.0482],[580,-0.0486],[581,-0.0484],[582,-0.0486],[583,-0.0484],[585,-0.0486],[587,-0.0482],[589,-0.0482],[590,-0.0476],[591,-0.0476],[595,-0.0461],[596,-0.0459],[611,-0.0459],[612,-0.0434],[616,-0.0432],[619,-0.0429],[624,-0.0431],[627,-0.0434],[629,-0.0432],[630,-0.0443],[633,-0.0443],[634,-0.044],[676,-0.1025],[677,-0.1027],[684,-0.1052],[686,-0.1052],[687,-0.1082],[689,-0.1082],[690,-0.1112],[691,-0.1112],[692,-0.1115],[694,-0.1125],[697,-0.1136],[706,-0.1139],[707,-0.1136],[711,-0.1137],[712,-0.114],[716,-0.114],[717,-0.1118],[724,-0.1118],[725,-0.112],[726,-0.1115],[728,-0.1114],[729,-0.111],[731,-0.1108],[732,-0.1104],[734,-0.1102],[735,-0.1096],[736,-0.1096],[737,-0.109],[738,-0.1089],[739,-0.1084],[740,-0.1083],[750,-0.1051],[753,-0.1047],[754,-0.1045],[759,-0.1041],[770,-0.0459],[771,-0.0459],[772,-0.0434],[776,-0.0432],[779,-0.0429],[784,-0.0431],[787,-0.0434],[789,-0.0432],[790,-0.0443],[793,-0.0443],[794,-0.044],[795,-0.044],[796,-0.0453],[799,-0.0453],[800,-0.045],[801,-0.0463],[803,-0.0463],[835,-0.102],[837,-0.1025],[887,-0.044],[888,-0.0453],[891,-0.0453],[892,-0.045],[893,-0.0463],[895,-0.0463],[927,-0.102],[929,-0.1025]]
       }
       spine_04_ctrl.Location.Z {
        type=MovieSceneScriptingFloatChannel
        default=0
        thinned_from=735
        keys [[-2,0.001],[8,0.0008],[9,0.0042],[11,0.0042],[12,0.0076],[13,0.0076],[14,0.0113],[15,0.0124],[18,0.0198],[19,0.0248],[28,0.0474],[29,0.0474],[30,0.0499],[31,0.0493],[33,0.0493],[34,0.0533],[38,0.0533],[46,0.0535],[47,0.0523],[48,0.0524],[49,0.0503],[51,0.0504],[52,0.0483],[54,0.0483],[55,0.0463],[56,0.0463],[57,0.0443],[58,0.0443],[60,0.0402],[61,0.0402],[62,0.0382],[66,0.0338],[67,0.0316],[72,0.0261],[73,0.0261],[76,0.0228],[77,0.0228],[78,0.0217],[79,0.0209],[80,0.0209],[81,0.0196],[82,0.0196],[83,0.0182],[84,0.0183],[85,0.017],[86,0.017],[87,0.0158],[88,0.0131],[91,0.0094],[92,0.0069],[93,0.0057],[94,0.0031],[95,0.002],[96,-0.0046],[97,-0.0043],[98,-0.0109],[99,-0.0107],[100,-0.0172],[101,-0.017],[102,-0.0236],[103,-0.0299],[104,-0.0297],[105,-0.0363],[106,-0.0363],[107,-0.036],[108,-0.0426],[109,-0.0424],[110,-0.0424],[111,-0.0429],[115,-0.0427],[116,-0.0374],[117,-0.0374],[118,-0.0323],[119,-0.0323],[121,-0.0218],[122,-0.022],[124,-0.0117],[125,-0.0118],[126,-0.0067],[127,-0.0046],[128,0.0015],[129,0.001],[130,0.001],[131,0.0065],[133,0.0065],[134,0.0121],[141,0.0121],[142,0.0059],[143,0.0088],[144,0.0088],[145,0.0028],[146,0.0031],[149,0.0031],[150,-0.0025],[154,-0.0025],[155,-0.0082],[158,-0.0082],[159,-0.0086],[174,-0.0086],[175,-0.0079],[176,-0.0079],[177,-0.0081],[178,-0.0041],[180,-0.0041],[181,-0.0043],[182,-0.0043],[183,-0.0003],[185,-0.0003],[186,-0.0005],[188,-0.0005],[189,0.0035],[190,0.0035],[191,0.0029],[196,0.0029],[197,0.0069],[201,0.0069],[202,0.0073],[206,0.0073],[207,0.0063],[210,0.0063],[211,0.0054],[212,0.0058],[214,0.0058],[215,0.0048],[218,0.0048],[219,0.0053],[220,0.0062],[221,0.0062],[222,0.0057],[223,0.0065],[224,0.0068],[225,0.0036],[226,0.0036],[228,0.0042],[229,0.0006],[230,0.001],[231,0.001],[232,-0.0023],[233,-0.0058],[234,-0.0058],[235,-0.0093],[236,-0.009],[238,-0.0161],[239,-0.0178],[254,-0.0582],[255,-0.0588],[259,-0.0681],[260,-0.0681],[262,-0.0727],[264,-0.0727],[265,-0.075],[270,-0.0748],[271,-0.0741],[286,-0.0741],[287,-0.0707],[292,-0.0707],[293,-0.072],[294,-0.072],[295,-0.0733],[296,-0.0733],[297,-0.0747],[298,-0.0747],[299,-0.076],[300,-0.076],[302,-0.0787],[303,-0.0781],[304,-0.0781],[305,-0.0816],[308,-0.0816],[309,-0.0851],[323,0.001],[333,0.0008],[334,0.0042],[336,0.0042],[337,0.0076],[338,0.0076],[339,0.0113],[340,0.0124],[343,0.0198],[344,0.0248],[353,0.0474],[354,0.0474],[355,0.0499],[356,0.0493],[358,0.0493],[359,0.0533],[363,0.0533],[371,0.0535],[372,0.0523],[373,0.0524],[374,0.0503],[376,0.0504],[377,0.0483],[379,0.0483],[380,0.0463],[381,0.0463],[382,0.0443],[383,0.0443],[385,0.0402],[386,0.0402],[387,0.0382],[391,0.0338],[392,0.0316],[397,0.0261],[398,0.0261],[401,0.0228],[402,0.0228],[403,0.0217],[404,0.0209],[405,0.0209],[406,0.0196],[407,0.0196],[408,0.0182],[409,0.0183],[410,0.017],[411,0.017],[412,0.0158],[413,0.0131],[416,0.0094],[417,0.0069],[418,0.0057],[419,0.0031],[420,0.002],[421,-0.0046],[422,-0.0043],[423,-0.0109],[424,-0.0107],[425,-0.0172],[426,-0.017],[427,-0.0236],[428,-0.0299],[429,-0.0297],[430,-0.0363],[431,-0.0363],[432,-0.036],[433,-0.0426],[434,-0.0424],[435,-0.0424],[436,-0.0429],[440,-0.0427],[441,-0.0374],[442,-0.0374],[443,-0.0323],[444,-0.0323],[446,-0.0218],[447,-0.022],[449,-0.0117],[450,-0.0118],[451,-0.0067],[452,-0.0046],[453,0.0015],[454,0.001],[455,0.001],[456,0.0065],[458,0.0065],[459,0.0121],[466,0.0121],[467,0.0059],[468,0.0088],[469,0.0088],[470,0.0028],[471,0.0031],[474,0.0031],[475,-0.0025],[479,-0.0025],[480,-0.0082],[483,-0.0082],[484,-0.0086],[499,-0.0086],[500,-0.0079],[501,-0.0079],[502,-0.0081],[503,-0.0041],[505,-0.0041],[506,-0.0043],[507,-0.0043],[508,-0.0003],[510,-0.0003],[511,-0.0005],[513,-0.0005],[514,0.0035],[515,0.0035],[516,0.0029],[521,0.0029],[522,0.0069],[526,0.0069],[527,0.0073],[531,0.0073],[532,0.0063],[535,0.0063],[536,0.0054],[537,0.0058],[539,0.0058],[540,0.0048],[543,0.0048],[544,0.0053],[545,0.0062],[546,0.0062],[547,0.0057],[548,0.0065],[549,0.0068],[550,0.0036],[551,0.0036],[553,0.0042],[554,0.0006],[555,0.001],[556,0.001],[557,-0.0023],[558,-0.0058],[559,-0.0058],[560,-0.0093],[561,-0.009],[563,-0.0161],[564,-0.0178],[579,-0.0582],[580,-0.0588],[584,-0.0681],[585,-0.0681],[587,-0.0727],[589,-0.0727],[590,-0.075],[595,-0.0748],[596,-0.0741],[611,-0.0741],[612,-0.0707],[617,-0.0707],[618,-0.072],[619,-0.072],[620,-0.0733],[621,-0.0733],[622,-0.0747],[623,-0.0747],[624,-0.076],[625,-0.076],[627,-0.0787],[628,-0.0781],[629,-0.0781],[630,-0.0816],[633,-0.0816],[634,-0.0851],[676,0.001],[686,0.0008],[687,0.0042],[689,0.0042],[690,0.0076],[691,0.0076],[692,0.0113],[693,0.0124],[696,0.0198],[697,0.0248],[706,0.0474],[707,0.0474],[708,0.0499],[709,0.0493],[711,0.0493],[712,0.0533],[716,0.0533],[724,0.0535],[725,0.0523],[726,0.0524],[727,0.0503],[729,0.0504],[730,0.0483],[732,0.0483],[733,0.0463],[734,0.0463],[735,0.0443],[736,0.0443],[738,0.0402],[739,0.0402],[740,0.0382],[744,0.0338],[745,0.0316],[750,0.0261],[751,0.0261],[754,0.0228],[755,0.0228],[757,0.0209],[759,0.0196],[770,-0.0741],[771,-0.0741],[772,-0.0707],[777,-0.0707],[778,-0.072],[779,-0.072],[780,-0.0733],[781,-0.0733],[782,-0.0747],[783,-0.0747],[784,-0.076],[785,-0.076],[787,-0.0787],[788,-0.0781],[789,-0.0781],[790,-0.0816],[793,-0.0816],[794,-0.0851],[799,-0.0852],[800,-0.0886],[803,-0.0887],[835,0.0001],[837,0.001],[887,-0.0851],[891,-0.0852],[892,-0.0886],[895,-0.0887],[927,0.0001],[929,0.001]]
       }
       spine_04_ctrl.Rotation.X {
        type=MovieSceneScriptingFloatChannel
        default=0
        thinned_from=693
        keys [[-2,2.1282],[-1,2.1094],[6,1.6757],[8,1.6753],[9,1.6554],[10,1.6103],[11,1.5519],[12,1.481],[13,1.401],[14,1.3127],[16,1.1204],[20,0.7195],[21,0.6215],[22,0.5306],[23,0.4499],[24,0.3804],[26,0.2535],[28,0.1241],[29,0.0621],[30,-0.0006],[31,-0.0621],[32,-0.1232],[33,-0.1803],[34,-0.2424],[35,-0.3005],[37,-0.4092],[39,-0.5197],[42,-0.6755],[44,-0.7725],[46,-0.8623],[48,-0.9453],[49,-0.9843],[51,-1.0517],[52,-1.08],[53,-1.1028],[54,-1.1215],[55,-1.1361],[57,-1.1577],[60,-1.1777],[64,-1.1897],[69,-1.1905],[75,-1.1825],[77,-1.1874],[78,-1.1934],[79,-1.2034],[81,-1.2389],[82,-1.2465],[83,-1.2367],[85,-1.183],[86,-1.1699],[87,-1.1698],[88,-1.1804],[89,-1.2008],[90,-1.2287],[91,-1.2632],[94,-1.3767],[95,-1.4104],[97,-1.4647],[98,-1.4867],[99,-1.504],[100,-1.5159],[101,-1.5204],[102,-1.5108],[103,-1.4831],[104,-1.4343],[105,-1.3655],[106,-1.2777],[107,-1.1687],[110,-0.7986],[111,-0.6723],[112,-0.5482],[113,-0.4185],[114,-0.2854],[115,-0.1413],[116,0.0066],[117,0.1552],[118,0.3019],[119,0.4415],[120,0.5738],[121,0.6987],[122,0.814],[123,0.9194],[124,1.0174],[125,1.103],[126,1.1861],[127,1.2628],[129,1.4033],[131,1.527],[132,1.5844],[133,1.6334],[134,1.6778],[135,1.7136],[136,1.7324],[137,1.7463],[138,1.7475],[139,1.7403],[140,1.7288],[141,1.7089],[142,1.6807],[143,1.6489],[144,1.6107],[150,1.3418],[152,1.2586],[154,1.1819],[156,1.1126],[159,1.0149],[160,0.9778],[161,0.9371],[162,0.9032],[163,0.8842],[164,0.8747],[165,0.876],[166,0.8841],[167,0.8978],[168,0.9171],[170,0.9658],[173,1.0524],[174,1.0774],[176,1.1185],[178,1.1356],[180,1.1189],[181,1.096],[182,1.0663],[183,1.0306],[184,0.99],[185,0.9422],[186,0.8886],[187,0.8306],[188,0.7678],[189,0.7009],[190,0.6303],[191,0.5571],[192,0.4802],[193,0.4008],[194,0.3194],[195,0.2367],[196,0.1526],[197,0.0691],[198,-0.0151],[199,-0.0986],[200,-0.1823],[201,-0.2641],[202,-0.3454],[203,-0.4248],[204,-0.502],[205,-0.5769],[206,-0.6492],[207,-0.719],[208,-0.7862],[209,-0.8502],[210,-0.9112],[211,-0.9681],[212,-1.0218],[213,-1.0704],[214,-1.1064],[215,-1.1267],[218,-1.1758],[219,-1.1876],[220,-1.1939],[221,-1.1955],[223,-1.1844],[224,-1.1819],[225,-1.1848],[226,-1.1957],[227,-1.2178],[228,-1.2532],[229,-1.3056],[230,-1.3789],[231,-1.4765],[232,-1.5934],[233,-1.7191],[235,-1.9759],[236,-2.0984],[238,-2.3329],[239,-2.4635],[240,-2.6076],[242,-2.9279],[243,-3.0808],[244,-3.2261],[245,-3.3634],[247,-3.6168],[252,-4.1964],[255,-4.5241],[257,-4.7643],[259,-5.037],[262,-5.4746],[264,-5.7405],[266,-5.9727],[272,-6.5374],[280,-7.178],[283,-7.3728],[287,-7.5631],[291,-7.6694],[309,-7.6516],[353,0.1241],[354,0.0621],[355,-0.0006],[356,-0.0621],[357,-0.1232],[358,-0.1803],[359,-0.2424],[360,-0.3005],[362,-0.4092],[364,-0.5197],[367,-0.6755],[369,-0.7725],[371,-0.8623],[373,-0.9453],[374,-0.9843],[376,-1.0517],[377,-1.08],[378,-1.1028],[379,-1.1215],[380,-1.1361],[382,-1.1577],[385,-1.1777],[389,-1.1897],[394,-1.1905],[400,-1.1825],[402,-1.1874],[403,-1.1934],[404,-1.2034],[406,-1.2389],[407,-1.2465],[408,-1.2367],[410,-1.183],[411,-1.1699],[412,-1.1698],[413,-1.1804],[414,-1.2008],[415,-1.2287],[416,-1.2632],[419,-1.3767],[420,-1.4104],[422,-1.4647],[423,-1.4867],[424,-1.504],[425,-1.5159],[426,-1.5204],[427,-1.5108],[428,-1.4831],[429,-1.4343],[430,-1.3655],[431,-1.2777],[432,-1.1687],[435,-0.7986],[436,-0.6723],[437,-0.5482],[438,-0.4185],[439,-0.2854],[440,-0.1413],[441,0.0066],[442,0.1552],[443,0.3019],[444,0.4415],[445,0.5738],[446,0.6987],[447,0.814],[448,0.9194],[449,1.0174],[450,1.103],[451,1.1861],[452,1.2628],[454,1.4033],[456,1.527],[457,1.5844],[458,1.6334],[459,1.6778],[460,1.7136],[461,1.7324],[462,1.7463],[463,1.7475],[464,1.7403],[465,1.7288],[466,1.7089],[467,1.6807],[468,1.6489],[469,1.6107],[475,1.3418],[477,1.2586],[479,1.1819],[481,1.1126],[484,1.0149],[485,0.9778],[486,0.9371],[487,0.9032],[488,0.8842],[489,0.8747],[490,0.876],[491,0.8841],[492,0.8978],[493,0.9171],[495,0.9658],[498,1.0524],[499,1.0774],[501,1.1185],[503,1.1356],[505,1.1189],[506,1.096],[507,1.0663],[508,1.0306],[509,0.99],[510,0.9422],[511,0.8886],[512,0.8306],[513,0.7678],[514,0.7009],[515,0.6303],[516,0.5571],[517,0.4802],[518,0.4008],[519,0.3194],[520,0.2367],[521,0.1526],[522,0.0691],[523,-0.0151],[524,-0.0986],[525,-0.1823],[526,-0.2641],[527,-0.3454],[528,-0.4248],[529,-0.502],[530,-0.5769],[531,-0.6492],[532,-0.719],[533,-0.7862],[534,-0.8502],[535,-0.9112],[536,-0.9681],[537,-1.0218],[538,-1.0704],[539,-1.1064],[540,-1.1267],[543,-1.1758],[544,-1.1876],[545,-1.1939],[546,-1.1955],[548,-1.1844],[549,-1.1819],[550,-1.1848],[551,-1.1957],[552,-1.2178],[553,-1.2532],[554,-1.3056],[555,-1.3789],[556,-1.4765],[557,-1.5934],[558,-1.7191],[560,-1.9759],[561,-2.0984],[563,-2.3329],[564,-2.4635],[565,-2.6076],[567,-2.9279],[568,-3.0808],[569,-3.2261],[570,-3.3634],[572,-3.6168],[577,-4.1964],[580,-4.5241],[582,-4.7643],[584,-5.037],[587,-5.4746],[589,-5.7405],[591,-5.9727],[597,-6.5374],[605,-7.178],[608,-7.3728],[612,-7.5631],[616,-7.6694],[634,-7.6516],[676,-1.0286],[677,-1.0474],[684,-1.4811],[686,-1.4815],[687,-1.5014],[688,-1.5465],[689,-1.6049],[690,-1.6757],[691,-1.7557],[692,-1.8441],[694,-2.0364],[699,-2.5353],[700,-2.6262],[701,-2.7069],[702,-2.7764],[710,-3.28],[713,-3.4573],[720,-3.8322],[724,-4.0191],[727,-4.1411],[730,-4.2368],[733,-4.2929],[743,-4.3478],[746,-4.3481],[770,-6.1021],[777,-6.0143],[803,-5.4137],[835,4.9519],[837,4.9734],[847,5.1916],[895,-7.7047],[927,2.1068],[929,2.1282]]
       }
       spine_04_ctrl.Rotation.Y {
        type=MovieSceneScriptingFloatChannel
        default=0
        thinned_from=717
        keys [[-2,1.8646],[-1,1.866],[6,1.9085],[7,1.9122],[8,1.9078],[9,1.8893],[10,1.8537],[11,1.8033],[12,1.7387],[13,1.6615],[14,1.5715],[15,1.4689],[16,1.3574],[17,1.2355],[18,1.1067],[19,0.9734],[21,0.7029],[22,0.5731],[23,0.4494],[24,0.3379],[25,0.237],[26,0.1413],[27,0.0561],[28,-0.0178],[29,-0.0836],[30,-0.1417],[31,-0.1929],[32,-0.2358],[33,-0.2759],[35,-0.3476],[36,-0.379],[38,-0.437],[39,-0.4614],[40,-0.4836],[41,-0.5035],[43,-0.5391],[44,-0.5524],[46,-0.5742],[47,-0.5836],[48,-0.5881],[50,-0.5884],[51,-0.5725],[52,-0.5601],[53,-0.5436],[55,-0.4942],[56,-0.4653],[57,-0.4284],[58,-0.3953],[59,-0.3543],[60,-0.3172],[61,-0.2722],[62,-0.2311],[66,-0.0514],[67,-0.0079],[68,0.0348],[69,0.0763],[70,0.1156],[71,0.1529],[72,0.1877],[73,0.2192],[74,0.2481],[75,0.274],[76,0.2973],[77,0.3183],[78,0.3368],[79,0.3524],[80,0.3649],[81,0.3756],[82,0.3885],[83,0.405],[84,0.4154],[85,0.4195],[86,0.4227],[87,0.4289],[88,0.4286],[89,0.4176],[90,0.4015],[91,0.3838],[92,0.3689],[93,0.3605],[94,0.3633],[95,0.3823],[96,0.427],[97,0.5061],[98,0.6113],[99,0.7364],[100,0.8801],[101,1.0424],[102,1.2236],[103,1.4212],[104,1.634],[108,2.529],[109,2.7478],[110,2.9608],[111,3.1645],[112,3.3586],[113,3.5426],[114,3.7084],[115,3.8511],[116,3.9651],[117,4.0544],[118,4.1191],[119,4.1596],[120,4.1733],[121,4.1671],[122,4.1414],[123,4.0971],[124,4.0347],[125,3.9566],[126,3.8672],[130,3.4632],[131,3.3732],[132,3.2926],[133,3.2243],[134,3.1677],[135,3.1253],[136,3.097],[137,3.0841],[139,3.0869],[142,3.1242],[147,3.246],[150,3.328],[153,3.3946],[157,3.4436],[159,3.4497],[167,3.3835],[170,3.3344],[172,3.281],[174,3.2072],[176,3.1089],[178,2.9888],[185,2.5525],[187,2.4382],[189,2.3333],[191,2.2404],[193,2.1636],[195,2.1006],[197,2.0495],[199,2.0126],[201,1.9902],[203,1.9825],[204,1.9843],[205,1.9902],[206,2.0002],[207,2.0148],[208,2.0358],[209,2.0627],[210,2.0956],[211,2.1347],[212,2.1827],[213,2.2371],[214,2.2789],[215,2.2957],[216,2.2864],[217,2.2522],[218,2.1996],[219,2.136],[220,2.068],[221,2.0133],[222,1.9676],[224,1.8913],[225,1.859],[226,1.8317],[227,1.8103],[228,1.7972],[229,1.7934],[230,1.802],[231,1.825],[232,1.8687],[233,1.9332],[234,2.0132],[235,2.112],[236,2.2269],[237,2.3575],[238,2.5024],[239,2.6558],[241,2.9826],[242,3.1565],[243,3.3394],[245,3.7257],[247,4.1225],[248,4.3065],[249,4.4762],[250,4.63],[251,4.7648],[252,4.8805],[253,4.9751],[254,5.0478],[256,5.1572],[268,5.5749],[271,5.6406],[273,5.6591],[295,5.6511],[299,5.6006],[304,5.4684],[309,5.306],[331,1.9085],[332,1.9122],[333,1.9078],[334,1.8893],[335,1.8537],[336,1.8033],[337,1.7387],[338,1.6615],[339,1.5715],[340,1.4689],[341,1.3574],[342,1.2355],[343,1.1067],[344,0.9734],[346,0.7029],[347,0.5731],[348,0.4494],[349,0.3379],[350,0.237],[351,0.1413],[352,0.0561],[353,-0.0178],[354,-0.0836],[355,-0.1417],[356,-0.1929],[357,-0.2358],[358,-0.2759],[360,-0.3476],[361,-0.379],[363,-0.437],[364,-0.4614],[365,-0.4836],[366,-0.5035],[368,-0.5391],[369,-0.5524],[371,-0.5742],[372,-0.5836],[373,-0.5881],[375,-0.5884],[376,-0.5725],[377,-0.5601],[378,-0.5436],[380,-0.4942],[381,-0.4653],[382,-0.4284],[383,-0.3953],[384,-0.3543],[385,-0.3172],[386,-0.2722],[387,-0.2311],[391,-0.0514],[392,-0.0079],[393,0.0348],[394,0.0763],[395,0.1156],[396,0.1529],[397,0.1877],[398,0.2192],[399,0.2481],[400,0.274],[401,0.2973],[402,0.3183],[403,0.3368],[404,0.3524],[405,0.3649],[406,0.3756],[407,0.3885],[408,0.405],[409,0.4154],[410,0.4195],[411,0.4227],[412,0.4289],[413,0.4286],[414,0.4176],[415,0.4015],[416,0.3838],[417,0.3689],[418,0.3605],[419,0.3633],[420,0.3823],[421,0.427],[422,0.5061],[423,0.6113],[424,0.7364],[425,0.8801],[426,1.0424],[427,1.2236],[428,1.4212],[429,1.634],[433,2.529],[434,2.7478],[435,2.9608],[436,3.1645],[437,3.3586],[438,3.5426],[439,3.7084],[440,3.8511],[441,3.9651],[442,4.0544],[443,4.1191],[444,4.1596],[445,4.1733],[446,4.1671],[447,4.1414],[448,4.0971],[449,4.0347],[450,3.9566],[451,3.8672],[455,3.4632],[456,3.3732],[457,3.2926],[458,3.2243],[459,3.1677],[460,3.1253],[461,3.097],[462,3.0841],[464,3.0869],[467,3.1242],[472,3.246],[475,3.328],[478,3.3946],[482,3.4436],[484,3.4497],[492,3.3835],[495,3.3344],[497,3.281],[499,3.2072],[501,3.1089],[503,2.9888],[510,2.5525],[512,2.4382],[514,2.3333],[516,2.2404],[518,2.1636],[520,2.1006],[522,2.0495],[524,2.0126],[526,1.9902],[528,1.9825],[529,1.9843],[530,1.9902],[531,2.0002],[532,2.0148],[533,2.0358],[534,2.0627],[535,2.0956],[536,2.1347],[537,2.1827],[538,2.2371],[539,2.2789],[540,2.2957],[541,2.2864],[542,2.2522],[543,2.1996],[544,2.136],[545,2.068],[546,2.0133],[547,1.9676],[549,1.8913],[550,1.859],[551,1.8317],[552,1.8103],[553,1.7972],[554,1.7934],[555,1.802],[556,1.825],[557,1.8687],[558,1.9332],[559,2.0132],[560,2.112],[561,2.2269],[562,2.3575],[563,2.5024],[564,2.6558],[566,2.9826],[567,3.1565],[568,3.3394],[570,3.7257],[572,4.1225],[573,4.3065],[574,4.4762],[575,4.63],[576,4.7648],[577,4.8805],[578,4.9751],[579,5.0478],[581,5.1572],[593,5.5749],[596,5.6406],[598,5.6591],[620,5.6511],[624,5.6006],[629,5.4684],[634,5.306],[676,1.8646],[677,1.866],[684,1.9085],[685,1.9122],[686,1.9078],[687,1.8893],[688,1.8537],[689,1.8033],[690,1.7387],[691,1.6615],[692,1.5715],[693,1.4689],[694,1.3574],[695,1.2355],[696,1.1067],[697,0.9734],[699,0.7029],[700,0.5731],[701,0.4494],[702,0.3379],[703,0.237],[704,0.1413],[705,0.0561],[706,-0.0178],[707,-0.0836],[708,-0.1417],[709,-0.1929],[710,-0.2358],[711,-0.2759],[713,-0.3476],[714,-0.379],[716,-0.437],[717,-0.4614],[718,-0.4836],[719,-0.5035],[721,-0.5391],[722,-0.5524],[724,-0.5742],[725,-0.5836],[726,-0.5881],[728,-0.5884],[729,-0.5725],[730,-0.5601],[731,-0.5436],[733,-0.4942],[734,-0.4653],[735,-0.4284],[736,-0.3953],[737,-0.3543],[738,-0.3172],[739,-0.2722],[740,-0.2311],[741,-0.1861],[770,5.6707],[781,5.6445],[785,5.5797],[795,5.2796],[798,5.224],[800,5.2151],[803,5.231],[835,1.9058],[837,1.8646],[887,5.2796],[890,5.224],[892,5.2151],[895,5.231],[927,1.9058],[929,1.8646]]
       }
       spine_04_ctrl.Rotation.Z {
        type=MovieSceneScriptingFloatChannel
        default=0
        thinned_from=735
        keys [[-2,1.3514],[-1,1.3378],[6,0.8731],[7,0.7776],[8,0.679],[9,0.578],[10,0.475],[11,0.367],[13,0.1467],[14,0.0329],[16,-0.194],[17,-0.3091],[19,-0.5334],[20,-0.6417],[21,-0.747],[22,-0.8468],[23,-0.9414],[24,-1.0271],[25,-1.1052],[26,-1.1748],[27,-1.2367],[28,-1.2865],[29,-1.3268],[30,-1.3606],[31,-1.3856],[32,-1.406],[36,-1.4706],[38,-1.4966],[44,-1.5546],[48,-1.5766],[51,-1.5823],[57,-1.5635],[69,-1.4868],[74,-1.4687],[79,-1.4641],[81,-1.4679],[82,-1.483],[83,-1.5186],[84,-1.5833],[85,-1.6769],[86,-1.8001],[87,-1.9475],[88,-2.1238],[89,-2.3284],[90,-2.5522],[93,-3.2665],[94,-3.4909],[95,-3.6945],[96,-3.8674],[97,-4.0086],[98,-4.1198],[99,-4.2064],[100,-4.2677],[101,-4.3052],[102,-4.3177],[103,-4.3045],[104,-4.2654],[105,-4.2075],[106,-4.1398],[111,-3.7363],[113,-3.5874],[114,-3.505],[117,-3.2273],[118,-3.1466],[119,-3.0813],[120,-3.0375],[121,-3.0089],[122,-2.9912],[123,-2.9863],[124,-2.9997],[125,-3.0282],[126,-3.0684],[130,-3.2769],[132,-3.3673],[135,-3.4707],[136,-3.497],[138,-3.5617],[144,-3.6863],[156,-3.8405],[165,-3.8734],[172,-3.8592],[174,-3.869],[176,-3.9041],[180,-4.0196],[182,-4.0545],[183,-4.0624],[192,-3.9654],[197,-3.8688],[200,-3.7845],[203,-3.6751],[206,-3.5356],[208,-3.4252],[210,-3.3002],[211,-3.2325],[212,-3.1572],[213,-3.0682],[214,-2.9574],[215,-2.8236],[216,-2.6681],[217,-2.4888],[218,-2.2827],[219,-2.0408],[220,-1.7651],[221,-1.4608],[222,-1.1394],[223,-0.809],[225,-0.1353],[226,0.1986],[227,0.5221],[228,0.836],[229,1.1331],[230,1.4122],[231,1.6695],[232,1.8941],[233,2.0816],[234,2.2326],[235,2.3598],[236,2.4618],[237,2.5401],[238,2.5956],[239,2.6287],[240,2.6408],[241,2.6329],[242,2.6063],[243,2.5607],[244,2.4979],[245,2.4179],[246,2.3213],[247,2.2132],[248,2.0975],[252,1.6079],[253,1.4904],[254,1.381],[255,1.2782],[256,1.1804],[257,1.0881],[258,1.0028],[259,0.9242],[260,0.8531],[261,0.7884],[262,0.7314],[263,0.6824],[264,0.644],[265,0.6167],[266,0.6023],[267,0.601],[269,0.6137],[271,0.6282],[273,0.6465],[276,0.6758],[281,0.7133],[282,0.722],[283,0.7232],[284,0.742],[285,0.7443],[286,0.7651],[287,0.7684],[288,0.7896],[289,0.7933],[290,0.8145],[291,0.8249],[292,0.8321],[293,0.8356],[294,0.8359],[295,0.8335],[296,0.8247],[297,0.8082],[298,0.7839],[299,0.7533],[300,0.7175],[301,0.6759],[302,0.6308],[303,0.582],[304,0.5305],[305,0.4756],[306,0.4188],[307,0.3599],[308,0.2989],[309,0.2357],[323,1.3514],[324,1.3378],[331,0.8731],[332,0.7776],[333,0.679],[334,0.578],[335,0.475],[336,0.367],[338,0.1467],[339,0.0329],[341,-0.194],[342,-0.3091],[344,-0.5334],[345,-0.6417],[346,-0.747],[347,-0.8468],[348,-0.9414],[349,-1.0271],[350,-1.1052],[351,-1.1748],[352,-1.2367],[353,-1.2865],[354,-1.3268],[355,-1.3606],[356,-1.3856],[357,-1.406],[361,-1.4706],[363,-1.4966],[369,-1.5546],[373,-1.5766],[376,-1.5823],[382,-1.5635],[394,-1.4868],[399,-1.4687],[404,-1.4641],[406,-1.4679],[407,-1.483],[408,-1.5186],[409,-1.5833],[410,-1.6769],[411,-1.8001],[412,-1.9475],[413,-2.1238],[414,-2.3284],[415,-2.5522],[418,-3.2665],[419,-3.4909],[420,-3.6945],[421,-3.8674],[422,-4.0086],[423,-4.1198],[424,-4.2064],[425,-4.2677],[426,-4.3052],[427,-4.3177],[428,-4.3045],[429,-4.2654],[430,-4.2075],[431,-4.1398],[436,-3.7363],[438,-3.5874],[439,-3.505],[442,-3.2273],[443,-3.1466],[444,-3.0813],[445,-3.0375],[446,-3.0089],[447,-2.9912],[448,-2.9863],[449,-2.9997],[450,-3.0282],[451,-3.0684],[455,-3.2769],[457,-3.3673],[460,-3.4707],[461,-3.497],[463,-3.5617],[469,-3.6863],[481,-3.8405],[490,-3.8734],[497,-3.8592],[499,-3.869],[501,-3.9041],[505,-4.0196],[507,-4.0545],[508,-4.0624],[517,-3.9654],[522,-3.8688],[525,-3.7845],[528,-3.6751],[531,-3.5356],[533,-3.4252],[535,-3.3002],[536,-3.2325],[537,-3.1572],[538,-3.0682],[539,-2.9574],[540,-2.8236],[541,-2.6681],[542,-2.4888],[543,-2.2827],[544,-2.0408],[545,-1.7651],[546,-1.4608],[547,-1.1394],[548,-0.809],[550,-0.1353],[551,0.1986],[552,0.5221],[553,0.836],[554,1.1331],[555,1.4122],[556,1.6695],[557,1.8941],[558,2.0816],[559,2.2326],[560,2.3598],[561,2.4618],[562,2.5401],[563,2.5956],[564,2.6287],[565,2.6408],[566,2.6329],[567,2.6063],[568,2.5607],[569,2.4979],[570,2.4179],[571,2.3213],[572,2.2132],[573,2.0975],[577,1.6079],[578,1.4904],[579,1.381],[580,1.2782],[581,1.1804],[582,1.0881],[583,1.0028],[584,0.9242],[585,0.8531],[586,0.7884],[587,0.7314],[588,0.6824],[589,0.644],[590,0.6167],[591,0.6023],[592,0.601],[594,0.6137],[596,0.6282],[598,0.6465],[601,0.6758],[606,0.7133],[607,0.722],[608,0.7232],[609,0.742],[610,0.7443],[611,0.7651],[612,0.7684],[613,0.7896],[614,0.7933],[615,0.8145],[616,0.8249],[617,0.8321],[618,0.8356],[619,0.8359],[620,0.8335],[621,0.8247],[622,0.8082],[623,0.7839],[624,0.7533],[625,0.7175],[626,0.6759],[627,0.6308],[628,0.582],[629,0.5305],[630,0.4756],[631,0.4188],[632,0.3599],[633,0.2989],[634,0.2357],[676,1.3514],[677,1.3378],[684,0.8731],[685,0.7776],[686,0.679],[687,0.578],[688,0.475],[689,0.367],[691,0.1467],[692,0.0329],[694,-0.194],[695,-0.3091],[697,-0.5334],[698,-0.6417],[699,-0.747],[700,-0.8468],[701,-0.9414],[702,-1.0271],[703,-1.1052],[704,-1.1748],[705,-1.2367],[706,-1.2865],[707,-1.3268],[708,-1.3606],[709,-1.3856],[710,-1.406],[714,-1.4706],[716,-1.4966],[722,-1.5546],[726,-1.5766],[729,-1.5823],[735,-1.5635],[747,-1.4868],[752,-1.4687],[757,-1.4641],[759,-1.4679],[770,0.7443],[771,0.7651],[772,0.7684],[773,0.7896],[774,0.7933],[775,0.8145],[776,0.8249],[777,0.8321],[778,0.8356],[779,0.8359],[780,0.8335],[781,0.8247],[782,0.8082],[783,0.7839],[784,0.7533],[785,0.7175],[786,0.6759],[787,0.6308],[788,0.582],[789,0.5305],[790,0.4756],[791,0.4188],[792,0.3599],[793,0.2989],[794,0.2357],[796,0.1044],[797,0.0382],[798,-0.0307],[799,-0.1002],[800,-0.1707],[801,-0.2416],[802,-0.3064],[803,-0.3646],[835,1.3424],[837,1.3514],[887,0.1702],[888,0.1044],[889,0.0382],[890,-0.0307],[891,-0.1002],[892,-0.1707],[893,-0.2416],[894,-0.3064],[895,-0.3646],[927,1.3424],[929,1.3514]]
       }
       spine_04_ctrl.Scale.X {type=MovieSceneScriptingFloatChannel, default=1, const [1,735]}
       spine_04_ctrl.Scale.Y {type=MovieSceneScriptingFloatChannel, default=1, const [1,735]}
       spine_04_ctrl.Scale.Z {type=MovieSceneScriptingFloatChannel, default=1, const [1,735]}
       head_ctrl.Location.X {
        type=MovieSceneScriptingFloatChannel
        default=0
        thinned_from=751
        keys [[-2,0.2506],[-1,0.2513],[0,0.251],[1,0.2489],[4,0.2476],[5,0.2476],[6,0.247],[7,0.2445],[11,0.2415],[12,0.2393],[19,0.2328],[25,0.2282],[26,0.2313],[31,0.2299],[32,0.2306],[33,0.2305],[34,0.2313],[35,0.2312],[36,0.232],[37,0.2319],[38,0.2326],[39,0.2325],[40,0.2333],[41,0.2332],[42,0.234],[44,0.2338],[45,0.2346],[46,0.2346],[47,0.235],[58,0.2345],[59,0.231],[63,0.2309],[64,0.2304],[65,0.2304],[68,0.2287],[69,0.2287],[70,0.2282],[71,0.2282],[72,0.2276],[76,0.2276],[80,0.2303],[83,0.2328],[85,0.2331],[87,0.2316],[88,0.23],[90,0.2252],[93,0.2156],[95,0.2087],[96,0.2057],[97,0.2018],[98,0.1996],[99,0.1965],[100,0.195],[101,0.1927],[102,0.192],[103,0.1892],[104,0.1907],[105,0.1887],[106,0.1906],[107,0.19],[108,0.1912],[109,0.1932],[110,0.1927],[111,0.1941],[112,0.1942],[113,0.1961],[114,0.1961],[115,0.1979],[116,0.1979],[117,0.1991],[118,0.1988],[120,0.1997],[121,0.1983],[122,0.1978],[123,0.196],[124,0.1953],[125,0.1939],[126,0.1913],[131,0.1854],[133,0.1836],[134,0.1788],[135,0.1773],[142,0.1756],[150,0.1747],[151,0.1748],[152,0.1745],[155,0.1748],[156,0.1745],[158,0.1749],[159,0.1746],[160,0.1762],[163,0.1763],[164,0.1781],[165,0.177],[168,0.177],[169,0.1758],[170,0.1742],[171,0.1741],[172,0.1714],[173,0.1711],[174,0.1692],[175,0.1682],[177,0.1669],[178,0.1601],[179,0.1585],[181,0.1571],[183,0.1563],[190,0.1562],[191,0.1569],[197,0.1581],[198,0.1609],[199,0.1598],[205,0.1633],[210,0.1675],[211,0.1622],[213,0.1647],[214,0.1685],[215,0.1638],[216,0.1662],[217,0.1689],[218,0.1676],[219,0.1715],[220,0.1759],[221,0.1824],[222,0.1873],[223,0.1936],[224,0.2012],[225,0.2061],[226,0.2128],[227,0.2191],[228,0.2247],[229,0.2323],[231,0.2396],[232,0.2453],[236,0.2542],[237,0.2546],[238,0.2564],[239,0.2568],[240,0.2597],[241,0.2588],[243,0.2583],[244,0.2609],[245,0.2598],[248,0.2583],[249,0.2607],[250,0.2595],[264,0.2507],[267,0.2502],[277,0.2509],[278,0.2515],[283,0.2515],[284,0.252],[285,0.2518],[288,0.2528],[289,0.2546],[291,0.2548],[292,0.2565],[295,0.2564],[296,0.2581],[304,0.2551],[306,0.2538],[308,0.2532],[309,0.2524],[323,0.2506],[324,0.2513],[325,0.251],[326,0.2489],[329,0.2476],[330,0.2476],[331,0.247],[332,0.2445],[336,0.2415],[337,0.2393],[344,0.2328],[350,0.2282],[351,0.2313],[356,0.2299],[357,0.2306],[358,0.2305],[359,0.2313],[360,0.2312],[361,0.232],[362,0.2319],[363,0.2326],[364,0.2325],[365,0.2333],[366,0.2332],[367,0.234],[369,0.2338],[370,0.2346],[371,0.2346],[372,0.235],[383,0.2345],[384,0.231],[388,0.2309],[389,0.2304],[390,0.2304],[393,0.2287],[394,0.2287],[395,0.2282],[396,0.2282],[397,0.2276],[401,0.2276],[405,0.2303],[408,0.2328],[410,0.2331],[412,0.2316],[413,0.23],[415,0.2252],[418,0.2156],[420,0.2087],[421,0.2057],[422,0.2018],[423,0.1996],[424,0.1965],[425,0.195],[426,0.1927],[427,0.192],[428,0.1892],[429,0.1907],[430,0.1887],[431,0.1906],[432,0.19],[433,0.1912],[434,0.1932],[435,0.1927],[436,0.1941],[437,0.1942],[438,0.1961],[439,0.1961],[440,0.1979],[441,0.1979],[442,0.1991],[443,0.1988],[445,0.1997],[446,0.1983],[447,0.1978],[448,0.196],[449,0.1953],[450,0.1939],[451,0.1913],[456,0.1854],[458,0.1836],[459,0.1788],[460,0.1773],[467,0.1756],[475,0.1747],[476,0.1748],[477,0.1745],[480,0.1748],[481,0.1745],[483,0.1749],[484,0.1746],[485,0.1762],[488,0.1763],[489,0.1781],[490,0.177],[493,0.177],[494,0.1758],[495,0.1742],[496,0.1741],[497,0.1714],[498,0.1711],[499,0.1692],[500,0.1682],[502,0.1669],[503,0.1601],[504,0.1585],[506,0.1571],[508,0.1563],[515,0.1562],[516,0.1569],[522,0.1581],[523,0.1609],[524,0.1598],[530,0.1633],[535,0.1675],[536,0.1622],[538,0.1647],[539,0.1685],[540,0.1638],[541,0.1662],[542,0.1689],[543,0.1676],[544,0.1715],[545,0.1759],[546,0.1824],[547,0.1873],[548,0.1936],[549,0.2012],[550,0.2061],[551,0.2128],[552,0.2191],[553,0.2247],[554,0.2323],[556,0.2396],[557,0.2453],[561,0.2542],[562,0.2546],[563,0.2564],[564,0.2568],[565,0.2597],[566,0.2588],[568,0.2583],[569,0.2609],[570,0.2598],[573,0.2583],[574,0.2607],[575,0.2595],[589,0.2507],[592,0.2502],[602,0.2509],[603,0.2515],[608,0.2515],[609,0.252],[610,0.2518],[613,0.2528],[614,0.2546],[616,0.2548],[617,0.2565],[620,0.2564],[621,0.2581],[629,0.2551],[631,0.2538],[633,0.2532],[634,0.2524],[676,0.2506],[677,0.2513],[678,0.251],[679,0.2489],[682,0.2476],[683,0.2476],[684,0.247],[685,0.2445],[689,0.2415],[690,0.2393],[697,0.2328],[703,0.2282],[704,0.2313],[709,0.2299],[710,0.2306],[711,0.2305],[712,0.2313],[713,0.2312],[714,0.232],[715,0.2319],[716,0.2326],[717,0.2325],[718,0.2333],[719,0.2332],[720,0.234],[722,0.2338],[723,0.2346],[724,0.2346],[725,0.235],[736,0.2345],[737,0.231],[741,0.2309],[742,0.2304],[743,0.2304],[746,0.2287],[747,0.2287],[748,0.2282],[749,0.2282],[750,0.2276],[754,0.2276],[757,0.2294],[759,0.2313],[770,0.2518],[773,0.2528],[774,0.2546],[776,0.2548],[777,0.2565],[780,0.2564],[781,0.2581],[789,0.2551],[791,0.2538],[793,0.2532],[794,0.2524],[795,0.2504],[798,0.2491],[799,0.2468],[801,0.2458],[802,0.2441],[803,0.2437],[837,0.2506],[887,0.2504],[890,0.2491],[891,0.2468],[893,0.2458],[894,0.2441],[895,0.2437],[929,0.2506]]
       }
       head_ctrl.Location.Y {
        type=MovieSceneScriptingFloatChannel
        default=0
        thinned_from=751
        keys [[-2,-0.1298],[-1,-0.1354],[0,-0.1359],[1,-0.1372],[4,-0.1397],[5,-0.1453],[7,-0.1478],[8,-0.1486],[11,-0.1524],[12,-0.158],[19,-0.1669],[20,-0.1758],[25,-0.1809],[27,-0.1823],[28,-0.1836],[36,-0.1873],[37,-0.1874],[38,-0.1883],[40,-0.1889],[42,-0.1899],[43,-0.19],[45,-0.1909],[47,-0.191],[48,-0.1923],[58,-0.1929],[59,-0.1923],[61,-0.1923],[62,-0.1912],[63,-0.192],[67,-0.1918],[68,-0.19],[70,-0.1901],[72,-0.1898],[73,-0.188],[74,-0.188],[77,-0.183],[78,-0.1798],[79,-0.1779],[82,-0.1688],[83,-0.166],[84,-0.1637],[85,-0.1621],[86,-0.1614],[87,-0.1614],[88,-0.1628],[89,-0.1649],[90,-0.1681],[91,-0.1717],[92,-0.1762],[95,-0.1915],[96,-0.2004],[97,-0.2012],[98,-0.2091],[99,-0.2094],[100,-0.2166],[101,-0.2162],[102,-0.2233],[103,-0.2294],[104,-0.2285],[105,-0.2343],[106,-0.2333],[107,-0.2318],[108,-0.2379],[110,-0.2351],[112,-0.2347],[113,-0.2334],[114,-0.2334],[115,-0.2321],[117,-0.2311],[118,-0.2301],[119,-0.2302],[120,-0.2309],[121,-0.2308],[123,-0.2329],[124,-0.2335],[125,-0.2358],[126,-0.2375],[127,-0.2379],[133,-0.2428],[134,-0.2429],[135,-0.2559],[142,-0.2569],[144,-0.2593],[145,-0.2592],[146,-0.2604],[147,-0.2602],[148,-0.2613],[149,-0.2612],[150,-0.2623],[151,-0.2622],[152,-0.2632],[155,-0.2629],[156,-0.264],[158,-0.2639],[159,-0.2632],[163,-0.2632],[164,-0.2609],[165,-0.2625],[166,-0.2625],[167,-0.26],[168,-0.26],[169,-0.2617],[171,-0.2616],[172,-0.2605],[177,-0.2617],[178,-0.2613],[179,-0.2628],[180,-0.2631],[181,-0.2653],[190,-0.2653],[191,-0.2645],[192,-0.2644],[193,-0.2655],[198,-0.2648],[199,-0.266],[206,-0.2632],[209,-0.2611],[210,-0.2645],[212,-0.2625],[213,-0.2657],[214,-0.2624],[215,-0.2648],[217,-0.2617],[218,-0.257],[219,-0.2546],[220,-0.2475],[221,-0.2412],[222,-0.2331],[223,-0.2265],[224,-0.2173],[228,-0.1779],[229,-0.1673],[230,-0.1547],[231,-0.1456],[232,-0.137],[233,-0.129],[234,-0.1221],[235,-0.1158],[236,-0.1105],[237,-0.1095],[238,-0.105],[240,-0.1021],[241,-0.0993],[242,-0.0992],[243,-0.0994],[244,-0.0989],[245,-0.0971],[247,-0.0989],[248,-0.1],[249,-0.1004],[250,-0.0996],[254,-0.1058],[255,-0.107],[256,-0.1085],[257,-0.1103],[258,-0.1117],[259,-0.1125],[260,-0.1147],[261,-0.1152],[262,-0.1172],[263,-0.1175],[264,-0.1191],[265,-0.1197],[266,-0.121],[267,-0.1205],[268,-0.1214],[269,-0.1214],[270,-0.1224],[271,-0.1229],[273,-0.1227],[274,-0.1255],[277,-0.1251],[278,-0.1286],[281,-0.1283],[282,-0.1291],[283,-0.129],[284,-0.1316],[287,-0.1285],[288,-0.1291],[289,-0.1234],[290,-0.124],[291,-0.124],[292,-0.1184],[295,-0.1187],[296,-0.1133],[297,-0.1129],[298,-0.1134],[299,-0.1132],[301,-0.1147],[302,-0.1148],[303,-0.1153],[304,-0.1162],[305,-0.1155],[306,-0.1165],[307,-0.1205],[308,-0.1216],[309,-0.1209],[323,-0.1298],[324,-0.1354],[325,-0.1359],[326,-0.1372],[329,-0.1397],[330,-0.1453],[332,-0.1478],[333,-0.1486],[336,-0.1524],[337,-0.158],[344,-0.1669],[345,-0.1758],[350,-0.1809],[352,-0.1823],[353,-0.1836],[361,-0.1873],[362,-0.1874],[363,-0.1883],[365,-0.1889],[367,-0.1899],[368,-0.19],[370,-0.1909],[372,-0.191],[373,-0.1923],[383,-0.1929],[384,-0.1923],[386,-0.1923],[387,-0.1912],[388,-0.192],[392,-0.1918],[393,-0.19],[395,-0.1901],[397,-0.1898],[398,-0.188],[399,-0.188],[402,-0.183],[403,-0.1798],[404,-0.1779],[407,-0.1688],[408,-0.166],[409,-0.1637],[410,-0.1621],[411,-0.1614],[412,-0.1614],[413,-0.1628],[414,-0.1649],[415,-0.1681],[416,-0.1717],[417,-0.1762],[420,-0.1915],[421,-0.2004],[422,-0.2012],[423,-0.2091],[424,-0.2094],[425,-0.2166],[426,-0.2162],[427,-0.2233],[428,-0.2294],[429,-0.2285],[430,-0.2343],[431,-0.2333],[432,-0.2318],[433,-0.2379],[435,-0.2351],[437,-0.2347],[438,-0.2334],[439,-0.2334],[440,-0.2321],[442,-0.2311],[443,-0.2301],[444,-0.2302],[445,-0.2309],[446,-0.2308],[448,-0.2329],[449,-0.2335],[450,-0.2358],[451,-0.2375],[452,-0.2379],[458,-0.2428],[459,-0.2429],[460,-0.2559],[467,-0.2569],[469,-0.2593],[470,-0.2592],[471,-0.2604],[472,-0.2602],[473,-0.2613],[474,-0.2612],[475,-0.2623],[476,-0.2622],[477,-0.2632],[480,-0.2629],[481,-0.264],[483,-0.2639],[484,-0.2632],[488,-0.2632],[489,-0.2609],[490,-0.2625],[491,-0.2625],[492,-0.26],[493,-0.26],[494,-0.2617],[496,-0.2616],[497,-0.2605],[502,-0.2617],[503,-0.2613],[504,-0.2628],[505,-0.2631],[506,-0.2653],[515,-0.2653],[516,-0.2645],[517,-0.2644],[518,-0.2655],[523,-0.2648],[524,-0.266],[531,-0.2632],[534,-0.2611],[535,-0.2645],[537,-0.2625],[538,-0.2657],[539,-0.2624],[540,-0.2648],[542,-0.2617],[543,-0.257],[544,-0.2546],[545,-0.2475],[546,-0.2412],[547,-0.2331],[548,-0.2265],[549,-0.2173],[553,-0.1779],[554,-0.1673],[555,-0.1547],[556,-0.1456],[557,-0.137],[558,-0.129],[559,-0.1221],[560,-0.1158],[561,-0.1105],[562,-0.1095],[563,-0.105],[565,-0.1021],[566,-0.0993],[567,-0.0992],[568,-0.0994],[569,-0.0989],[570,-0.0971],[572,-0.0989],[573,-0.1],[574,-0.1004],[575,-0.0996],[579,-0.1058],[580,-0.107],[581,-0.1085],[582,-0.1103],[583,-0.1117],[584,-0.1125],[585,-0.1147],[586,-0.1152],[587,-0.1172],[588,-0.1175],[589,-0.1191],[590,-0.1197],[591,-0.121],[592,-0.1205],[593,-0.1214],[594,-0.1214],[595,-0.1224],[596,-0.1229],[598,-0.1227],[599,-0.1255],[602,-0.1251],[603,-0.1286],[606,-0.1283],[607,-0.1291],[608,-0.129],[609,-0.1316],[612,-0.1285],[613,-0.1291],[614,-0.1234],[615,-0.124],[616,-0.124],[617,-0.1184],[620,-0.1187],[621,-0.1133],[622,-0.1129],[623,-0.1134],[624,-0.1132],[626,-0.1147],[627,-0.1148],[628,-0.1153],[629,-0.1162],[630,-0.1155],[631,-0.1165],[632,-0.1205],[633,-0.1216],[634,-0.1209],[676,-0.1298],[677,-0.1354],[678,-0.1359],[679,-0.1372],[682,-0.1397],[683,-0.1453],[685,-0.1478],[686,-0.1486],[689,-0.1524],[690,-0.158],[697,-0.1669],[698,-0.1758],[703,-0.1809],[705,-0.1823],[706,-0.1836],[714,-0.1873],[715,-0.1874],[716,-0.1883],[718,-0.1889],[720,-0.1899],[721,-0.19],[723,-0.1909],[725,-0.191],[726,-0.1923],[736,-0.1929],[737,-0.1923],[739,-0.1923],[740,-0.1912],[741,-0.192],[745,-0.1918],[746,-0.19],[748,-0.1901],[750,-0.1898],[751,-0.188],[752,-0.188],[755,-0.183],[757,-0.1779],[759,-0.1718],[770,-0.1306],[772,-0.1285],[773,-0.1291],[774,-0.1234],[775,-0.124],[776,-0.124],[777,-0.1184],[780,-0.1187],[781,-0.1133],[782,-0.1129],[783,-0.1134],[784,-0.1132],[786,-0.1147],[787,-0.1148],[788,-0.1153],[789,-0.1162],[790,-0.1155],[791,-0.1165],[792,-0.1205],[793,-0.1216],[794,-0.1209],[796,-0.1232],[797,-0.1272],[798,-0.1283],[799,-0.1277],[801,-0.1296],[802,-0.1336],[803,-0.1344],[837,-0.1298],[887,-0.1222],[888,-0.1232],[889,-0.1272],[890,-0.1283],[891,-0.1277],[893,-0.1296],[894,-0.1336],[895,-0.1344],[929,-0.1298]]
       }
       head_ctrl.Location.Z {
        type=MovieSceneScriptingFloatChannel
        default=0
        thinned_from=751
        keys [[-2,-0.006],[-1,-0.0062],[0,-0.0055],[1,-0.0063],[4,-0.0046],[5,-0.0028],[7,-0.002],[8,-0.0002],[11,0.0013],[12,0.0048],[16,0.0073],[19,0.0084],[20,0.0105],[22,0.0106],[27,0.0092],[28,0.005],[30,0.0043],[31,0.0037],[32,0.0019],[33,0.0017],[34,-0.0001],[35,-0.0003],[36,-0.0021],[37,-0.0024],[38,-0.0041],[39,-0.0043],[40,-0.0061],[41,-0.0062],[42,-0.0079],[44,-0.0082],[45,-0.0098],[46,-0.0099],[47,-0.0114],[48,-0.0112],[54,-0.0106],[58,-0.0096],[59,-0.0063],[61,-0.0057],[62,-0.0056],[64,-0.0047],[65,-0.0029],[67,-0.0022],[68,-0.0009],[70,-0.0003],[71,0.0013],[72,0.0016],[73,0.0028],[74,0.003],[75,0.0042],[76,0.004],[77,0.0051],[78,0.0058],[79,0.006],[81,0.0075],[82,0.0085],[83,0.0097],[84,0.0113],[85,0.0132],[87,0.018],[89,0.0235],[93,0.0358],[94,0.0387],[95,0.0413],[96,0.0443],[97,0.045],[98,0.0478],[99,0.0472],[100,0.0498],[101,0.0491],[102,0.0494],[103,0.0507],[104,0.0497],[105,0.0498],[107,0.0454],[108,0.0454],[110,0.041],[111,0.0393],[112,0.0363],[113,0.035],[114,0.0318],[115,0.0303],[116,0.0265],[117,0.0225],[118,0.0209],[119,0.0168],[121,0.0134],[122,0.0092],[123,0.0078],[124,0.006],[125,0.0021],[126,0.0008],[127,-0.0019],[129,-0.0049],[131,-0.0074],[133,-0.0093],[134,-0.0094],[135,-0.007],[136,-0.0055],[138,-0.0056],[142,-0.0045],[143,-0.0041],[144,-0.0031],[145,-0.0026],[153,0.0044],[154,0.0049],[156,0.0065],[158,0.0072],[159,0.0081],[163,0.0087],[164,0.0079],[165,0.0013],[166,0.001],[167,0],[168,-0.0005],[169,-0.0077],[170,-0.0082],[171,-0.009],[172,-0.0168],[174,-0.0184],[175,-0.0198],[177,-0.0217],[178,-0.0216],[179,-0.0278],[180,-0.0285],[181,-0.0288],[190,-0.033],[191,-0.0336],[192,-0.034],[193,-0.034],[197,-0.0351],[198,-0.0357],[199,-0.0445],[206,-0.0456],[207,-0.0462],[209,-0.0462],[210,-0.0454],[211,-0.0443],[212,-0.0442],[213,-0.0432],[214,-0.0307],[215,-0.0288],[216,-0.0284],[217,-0.0276],[218,-0.0128],[219,-0.0108],[220,-0.0091],[221,0.0063],[222,0.0089],[223,0.0157],[224,0.0218],[225,0.0298],[227,0.0423],[228,0.047],[229,0.0529],[230,0.0563],[231,0.0607],[232,0.0632],[233,0.067],[234,0.0689],[235,0.0702],[236,0.0694],[237,0.0694],[238,0.0679],[239,0.0678],[240,0.0689],[241,0.0597],[244,0.0626],[245,0.0531],[248,0.0552],[249,0.0557],[250,0.0457],[255,0.0469],[256,0.047],[257,0.0448],[258,0.0449],[259,0.0446],[260,0.0426],[261,0.0422],[262,0.0401],[263,0.0395],[264,0.0372],[265,0.0367],[266,0.0342],[267,0.0333],[268,0.0308],[269,0.0303],[270,0.0278],[271,0.0268],[273,0.0263],[274,0.0202],[277,0.0199],[278,0.0143],[281,0.0143],[282,0.0148],[283,0.0148],[284,0.0089],[287,0.0079],[288,0.0064],[289,0.004],[290,0.0026],[291,0.0028],[292,0.0006],[295,0.0015],[296,-0.0006],[297,0.0015],[298,0.0019],[299,0.0041],[301,0.005],[302,0.0072],[303,0.0085],[304,0.009],[305,0.0131],[306,0.0136],[307,0.0155],[308,0.016],[309,0.0201],[323,-0.006],[324,-0.0062],[325,-0.0055],[326,-0.0063],[329,-0.0046],[330,-0.0028],[332,-0.002],[333,-0.0002],[336,0.0013],[337,0.0048],[341,0.0073],[344,0.0084],[345,0.0105],[347,0.0106],[352,0.0092],[353,0.005],[355,0.0043],[356,0.0037],[357,0.0019],[358,0.0017],[359,-0.0001],[360,-0.0003],[361,-0.0021],[362,-0.0024],[363,-0.0041],[364,-0.0043],[365,-0.0061],[366,-0.0062],[367,-0.0079],[369,-0.0082],[370,-0.0098],[371,-0.0099],[372,-0.0114],[373,-0.0112],[379,-0.0106],[383,-0.0096],[384,-0.0063],[386,-0.0057],[387,-0.0056],[389,-0.0047],[390,-0.0029],[392,-0.0022],[393,-0.0009],[395,-0.0003],[396,0.0013],[397,0.0016],[398,0.0028],[399,0.003],[400,0.0042],[401,0.004],[402,0.0051],[403,0.0058],[404,0.006],[406,0.0075],[407,0.0085],[408,0.0097],[409,0.0113],[410,0.0132],[412,0.018],[414,0.0235],[418,0.0358],[419,0.0387],[420,0.0413],[421,0.0443],[422,0.045],[423,0.0478],[424,0.0472],[425,0.0498],[426,0.0491],[427,0.0494],[428,0.0507],[429,0.0497],[430,0.0498],[432,0.0454],[433,0.0454],[435,0.041],[436,0.0393],[437,0.0363],[438,0.035],[439,0.0318],[440,0.0303],[441,0.0265],[442,0.0225],[443,0.0209],[444,0.0168],[446,0.0134],[447,0.0092],[448,0.0078],[449,0.006],[450,0.0021],[451,0.0008],[452,-0.0019],[454,-0.0049],[456,-0.0074],[458,-0.0093],[459,-0.0094],[460,-0.007],[461,-0.0055],[463,-0.0056],[467,-0.0045],[468,-0.0041],[469,-0.0031],[470,-0.0026],[478,0.0044],[479,0.0049],[481,0.0065],[483,0.0072],[484,0.0081],[488,0.0087],[489,0.0079],[490,0.0013],[491,0.001],[492,0],[493,-0.0005],[494,-0.0077],[495,-0.0082],[496,-0.009],[497,-0.0168],[499,-0.0184],[500,-0.0198],[502,-0.0217],[503,-0.0216],[504,-0.0278],[505,-0.0285],[506,-0.0288],[515,-0.033],[516,-0.0336],[517,-0.034],[518,-0.034],[522,-0.0351],[523,-0.0357],[524,-0.0445],[531,-0.0456],[532,-0.0462],[534,-0.0462],[535,-0.0454],[536,-0.0443],[537,-0.0442],[538,-0.0432],[539,-0.0307],[540,-0.0288],[541,-0.0284],[542,-0.0276],[543,-0.0128],[544,-0.0108],[545,-0.0091],[546,0.0063],[547,0.0089],[548,0.0157],[549,0.0218],[550,0.0298],[552,0.0423],[553,0.047],[554,0.0529],[555,0.0563],[556,0.0607],[557,0.0632],[558,0.067],[559,0.0689],[560,0.0702],[561,0.0694],[562,0.0694],[563,0.0679],[564,0.0678],[565,0.0689],[566,0.0597],[569,0.0626],[570,0.0531],[573,0.0552],[574,0.0557],[575,0.0457],[580,0.0469],[581,0.047],[582,0.0448],[583,0.0449],[584,0.0446],[585,0.0426],[586,0.0422],[587,0.0401],[588,0.0395],[589,0.0372],[590,0.0367],[591,0.0342],[592,0.0333],[593,0.0308],[594,0.0303],[595,0.0278],[596,0.0268],[598,0.0263],[599,0.0202],[602,0.0199],[603,0.0143],[606,0.0143],[607,0.0148],[608,0.0148],[609,0.0089],[612,0.0079],[613,0.0064],[614,0.004],[615,0.0026],[616,0.0028],[617,0.0006],[620,0.0015],[621,-0.0006],[622,0.0015],[623,0.0019],[624,0.0041],[626,0.005],[627,0.0072],[628,0.0085],[629,0.009],[630,0.0131],[631,0.0136],[632,0.0155],[633,0.016],[634,0.0201],[676,-0.006],[677,-0.0062],[678,-0.0055],[679,-0.0063],[682,-0.0046],[683,-0.0028],[685,-0.002],[686,-0.0002],[689,0.0013],[690,0.0048],[694,0.0073],[697,0.0084],[698,0.0105],[700,0.0106],[705,0.0092],[706,0.005],[708,0.0043],[709,0.0037],[710,0.0019],[711,0.0017],[712,-0.0001],[713,-0.0003],[714,-0.0021],[715,-0.0024],[716,-0.0041],[717,-0.0043],[718,-0.0061],[719,-0.0062],[720,-0.0079],[722,-0.0082],[723,-0.0098],[724,-0.0099],[725,-0.0114],[726,-0.0112],[732,-0.0106],[736,-0.0096],[737,-0.0063],[739,-0.0057],[740,-0.0056],[742,-0.0047],[743,-0.0029],[745,-0.0022],[746,-0.0009],[748,-0.0003],[749,0.0013],[750,0.0016],[751,0.0028],[752,0.003],[753,0.0042],[754,0.004],[755,0.0051],[757,0.006],[759,0.0075],[770,0.0085],[772,0.0079],[773,0.0064],[774,0.004],[775,0.0026],[776,0.0028],[777,0.0006],[780,0.0015],[781,-0.0006],[782,0.0015],[783,0.0019],[784,0.0041],[786,0.005],[787,0.0072],[788,0.0085],[789,0.009],[790,0.0131],[791,0.0136],[792,0.0155],[793,0.016],[794,0.0201],[796,0.021],[797,0.0229],[798,0.0233],[799,0.0273],[801,0.028],[802,0.0298],[803,0.0301],[837,-0.006],[888,0.021],[889,0.0229],[890,0.0233],[891,0.0273],[893,0.028],[894,0.0298],[895,0.0301],[929,-0.006]]
       }
       head_ctrl.Rotation.X {
        type=MovieSceneScriptingFloatChannel
        default=0
        thinned_from=751
        keys [[-2,-5.7506],[-1,-5.3693],[0,-5.2561],[1,-5.176],[2,-5.1261],[3,-5.1044],[4,-5.1087],[5,-5.1392],[6,-5.1867],[7,-5.2626],[8,-5.3614],[9,-5.4735],[10,-5.6053],[12,-5.9037],[16,-6.5295],[18,-6.796],[19,-6.9138],[21,-7.0911],[23,-7.2042],[25,-7.2423],[27,-7.2095],[28,-7.1671],[30,-7.0371],[31,-6.9529],[33,-6.7421],[35,-6.5027],[43,-5.4865],[45,-5.2714],[48,-5],[50,-4.8467],[51,-4.7907],[52,-4.7468],[53,-4.6932],[57,-4.5697],[61,-4.4867],[64,-4.4509],[66,-4.445],[68,-4.4731],[69,-4.5143],[70,-4.5662],[71,-4.6379],[72,-4.7273],[73,-4.8445],[74,-4.9927],[75,-5.1722],[76,-5.4146],[77,-5.734],[78,-6.1024],[79,-6.5142],[80,-6.9631],[84,-8.8483],[85,-9.2909],[86,-9.6959],[87,-10.0457],[89,-10.6073],[91,-11.1363],[92,-11.3839],[93,-11.6016],[94,-11.7905],[95,-11.9385],[96,-12.0133],[97,-12.0024],[98,-11.9288],[99,-11.795],[100,-11.6122],[101,-11.3861],[104,-10.5575],[106,-9.9856],[109,-9.2179],[111,-8.7376],[113,-8.3226],[114,-8.1505],[115,-8.0299],[116,-7.9572],[118,-7.9414],[119,-7.9886],[121,-8.1901],[123,-8.4954],[124,-8.6691],[127,-9.2717],[130,-9.8981],[132,-10.2781],[133,-10.4496],[134,-10.595],[135,-10.7074],[136,-10.7914],[138,-10.8528],[142,-10.8555],[155,-10.5182],[160,-10.2885],[161,-10.2673],[162,-10.2178],[163,-10.1048],[164,-9.9324],[166,-9.4927],[168,-8.9763],[169,-8.7342],[170,-8.5223],[171,-8.3549],[172,-8.2474],[173,-8.2024],[175,-8.2583],[178,-8.4593],[179,-8.5421],[182,-8.73],[185,-8.8668],[186,-8.8953],[187,-8.8513],[189,-8.8697],[190,-8.8629],[192,-8.8012],[194,-8.6885],[198,-8.3724],[202,-8.0075],[204,-7.851],[206,-7.7439],[207,-7.7102],[208,-7.6924],[209,-7.7072],[210,-7.7866],[211,-7.8931],[213,-8.1604],[214,-8.3321],[215,-8.5437],[216,-8.7959],[218,-9.3612],[220,-9.9675],[221,-10.2557],[222,-10.5224],[223,-10.7637],[224,-10.9745],[225,-11.1566],[226,-11.3075],[227,-11.4346],[228,-11.5334],[229,-11.6017],[230,-11.6351],[231,-11.6355],[232,-11.6012],[233,-11.5107],[234,-11.3848],[235,-11.215],[236,-11.0103],[237,-10.7527],[238,-10.4345],[239,-10.0553],[240,-9.6252],[241,-9.1451],[246,-6.4636],[247,-5.9437],[248,-5.4479],[249,-4.9754],[250,-4.5306],[251,-4.1182],[252,-3.7351],[253,-3.3834],[254,-3.0497],[255,-2.7342],[256,-2.4255],[258,-1.7893],[259,-1.4648],[261,-0.832],[262,-0.5206],[263,-0.2159],[264,0.0844],[265,0.3797],[266,0.6694],[267,0.9634],[269,1.5318],[271,2.0735],[272,2.3302],[273,2.5755],[274,2.809],[278,3.6968],[280,4.1183],[282,4.5017],[284,4.8405],[286,5.1551],[288,5.4508],[290,5.7196],[291,5.8251],[292,5.9098],[293,5.9668],[294,5.9944],[295,6.0051],[296,5.9897],[297,5.9418],[298,5.8699],[299,5.7766],[300,5.6661],[301,5.5381],[302,5.3923],[303,5.2354],[304,5.0673],[309,4.1445],[323,-5.7506],[324,-5.3693],[325,-5.2561],[326,-5.176],[327,-5.1261],[328,-5.1044],[329,-5.1087],[330,-5.1392],[331,-5.1867],[332,-5.2626],[333,-5.3614],[334,-5.4735],[335,-5.6053],[337,-5.9037],[341,-6.5295],[343,-6.796],[344,-6.9138],[346,-7.0911],[348,-7.2042],[350,-7.2423],[352,-7.2095],[353,-7.1671],[355,-7.0371],[356,-6.9529],[358,-6.7421],[360,-6.5027],[368,-5.4865],[370,-5.2714],[373,-5],[375,-4.8467],[376,-4.7907],[377,-4.7468],[378,-4.6932],[382,-4.5697],[386,-4.4867],[389,-4.4509],[391,-4.445],[393,-4.4731],[394,-4.5143],[395,-4.5662],[396,-4.6379],[397,-4.7273],[398,-4.8445],[399,-4.9927],[400,-5.1722],[401,-5.4146],[402,-5.734],[403,-6.1024],[404,-6.5142],[405,-6.9631],[409,-8.8483],[410,-9.2909],[411,-9.6959],[412,-10.0457],[414,-10.6073],[416,-11.1363],[417,-11.3839],[418,-11.6016],[419,-11.7905],[420,-11.9385],[421,-12.0133],[422,-12.0024],[423,-11.9288],[424,-11.795],[425,-11.6122],[426,-11.3861],[429,-10.5575],[431,-9.9856],[434,-9.2179],[436,-8.7376],[438,-8.3226],[439,-8.1505],[440,-8.0299],[441,-7.9572],[443,-7.9414],[444,-7.9886],[446,-8.1901],[448,-8.4954],[449,-8.6691],[452,-9.2717],[455,-9.8981],[457,-10.2781],[458,-10.4496],[459,-10.595],[460,-10.7074],[461,-10.7914],[463,-10.8528],[467,-10.8555],[480,-10.5182],[485,-10.2885],[486,-10.2673],[487,-10.2178],[488,-10.1048],[489,-9.9324],[491,-9.4927],[493,-8.9763],[494,-8.7342],[495,-8.5223],[496,-8.3549],[497,-8.2474],[498,-8.2024],[500,-8.2583],[503,-8.4593],[504,-8.5421],[507,-8.73],[510,-8.8668],[511,-8.8953],[512,-8.8513],[514,-8.8697],[515,-8.8629],[517,-8.8012],[519,-8.6885],[523,-8.3724],[527,-8.0075],[529,-7.851],[531,-7.7439],[532,-7.7102],[533,-7.6924],[534,-7.7072],[535,-7.7866],[536,-7.8931],[538,-8.1604],[539,-8.3321],[540,-8.5437],[541,-8.7959],[543,-9.3612],[545,-9.9675],[546,-10.2557],[547,-10.5224],[548,-10.7637],[549,-10.9745],[550,-11.1566],[551,-11.3075],[552,-11.4346],[553,-11.5334],[554,-11.6017],[555,-11.6351],[556,-11.6355],[557,-11.6012],[558,-11.5107],[559,-11.3848],[560,-11.215],[561,-11.0103],[562,-10.7527],[563,-10.4345],[564,-10.0553],[565,-9.6252],[566,-9.1451],[571,-6.4636],[572,-5.9437],[573,-5.4479],[574,-4.9754],[575,-4.5306],[576,-4.1182],[577,-3.7351],[578,-3.3834],[579,-3.0497],[580,-2.7342],[581,-2.4255],[583,-1.7893],[584,-1.4648],[586,-0.832],[587,-0.5206],[588,-0.2159],[589,0.0844],[590,0.3797],[591,0.6694],[592,0.9634],[594,1.5318],[596,2.0735],[597,2.3302],[598,2.5755],[599,2.809],[603,3.6968],[605,4.1183],[607,4.5017],[609,4.8405],[611,5.1551],[613,5.4508],[615,5.7196],[616,5.8251],[617,5.9098],[618,5.9668],[619,5.9944],[620,6.0051],[621,5.9897],[622,5.9418],[623,5.8699],[624,5.7766],[625,5.6661],[626,5.5381],[627,5.3923],[628,5.2354],[629,5.0673],[634,4.1445],[676,-5.7506],[677,-5.3693],[678,-5.2561],[679,-5.176],[680,-5.1261],[681,-5.1044],[682,-5.1087],[683,-5.1392],[684,-5.1867],[685,-5.2626],[686,-5.3614],[687,-5.4735],[688,-5.6053],[690,-5.9037],[694,-6.5295],[696,-6.796],[697,-6.9138],[699,-7.0911],[701,-7.2042],[703,-7.2423],[705,-7.2095],[706,-7.1671],[708,-7.0371],[709,-6.9529],[711,-6.7421],[713,-6.5027],[721,-5.4865],[723,-5.2714],[726,-5],[728,-4.8467],[729,-4.7907],[730,-4.7468],[731,-4.6932],[735,-4.5697],[739,-4.4867],[742,-4.4509],[744,-4.445],[746,-4.4731],[747,-4.5143],[748,-4.5662],[749,-4.6379],[750,-4.7273],[751,-4.8445],[752,-4.9927],[753,-5.1722],[754,-5.4146],[755,-5.734],[757,-6.5142],[759,-7.4331],[770,5.0004],[773,5.4508],[775,5.7196],[776,5.8251],[777,5.9098],[778,5.9668],[779,5.9944],[780,6.0051],[781,5.9897],[782,5.9418],[783,5.8699],[784,5.7766],[785,5.6661],[786,5.5381],[787,5.3923],[788,5.2354],[789,5.0673],[794,4.1445],[796,3.798],[797,3.6371],[798,3.4907],[799,3.3567],[800,3.2451],[801,3.1527],[802,3.0773],[803,3.0159],[837,-5.7506],[887,3.9675],[888,3.798],[889,3.6371],[890,3.4907],[891,3.3567],[892,3.2451],[893,3.1527],[894,3.0773],[895,3.0159],[929,-5.7506]]
       }
       head_ctrl.Rotation.Y {
        type=MovieSceneScriptingFloatChannel
        default=0
        thinned_from=751
        keys [[-2,-4.3117],[-1,-3.4371],[0,-3.0805],[1,-2.7387],[2,-2.4155],[3,-2.1057],[4,-1.8132],[5,-1.5327],[6,-1.2654],[7,-1.0114],[8,-0.7351],[9,-0.4007],[10,0.0085],[11,0.4682],[12,0.9722],[13,1.5186],[14,2.0984],[15,2.7007],[18,4.5803],[19,5.1965],[20,5.7862],[21,6.3423],[22,6.8525],[23,7.2913],[24,7.6628],[25,7.9697],[26,8.2364],[27,8.4664],[28,8.6489],[29,8.7942],[30,8.9023],[31,8.9823],[34,9.1164],[38,9.196],[45,9.1587],[49,9.0867],[50,9.0435],[51,9.0211],[54,8.8623],[57,8.6379],[66,7.7767],[71,7.2183],[73,6.9602],[75,6.665],[76,6.484],[77,6.265],[78,6.0198],[80,5.4756],[81,5.1901],[83,4.6003],[84,4.3351],[85,4.1482],[86,4.0503],[87,4.0319],[88,4.1458],[89,4.3656],[90,4.6501],[91,4.9799],[93,5.6639],[94,5.9786],[95,6.2358],[96,6.4087],[97,6.4781],[98,6.4535],[99,6.3643],[100,6.1971],[101,5.9501],[102,5.6313],[103,5.2144],[104,4.6984],[105,4.1038],[106,3.4618],[107,2.7769],[108,2.0674],[110,0.6307],[111,-0.0813],[112,-0.7708],[113,-1.4526],[114,-2.12],[115,-2.7617],[116,-3.3663],[117,-3.9127],[118,-4.402],[119,-4.817],[120,-5.1478],[121,-5.4115],[122,-5.6102],[123,-5.7475],[124,-5.8119],[125,-5.8239],[126,-5.7991],[127,-5.7535],[131,-5.4545],[132,-5.3875],[134,-5.2828],[137,-5.1634],[142,-4.995],[150,-4.715],[153,-4.6456],[155,-4.6271],[157,-4.6353],[160,-4.6954],[161,-4.7229],[162,-4.7848],[163,-4.9003],[164,-5.049],[165,-5.2282],[169,-6.0489],[170,-6.2447],[171,-6.4136],[172,-6.5397],[173,-6.6229],[174,-6.6775],[175,-6.7076],[176,-6.686],[177,-6.6401],[179,-6.5],[183,-6.1695],[184,-6.1048],[189,-5.6763],[191,-5.5239],[193,-5.392],[196,-5.2392],[198,-5.1798],[200,-5.1564],[206,-5.1896],[207,-5.1836],[208,-5.1536],[209,-5.0176],[210,-4.653],[211,-4.2348],[212,-3.7891],[213,-3.3346],[214,-2.8723],[215,-2.3949],[216,-1.9006],[217,-1.4337],[218,-1.0046],[219,-0.6509],[220,-0.3697],[221,-0.1862],[222,-0.106],[223,-0.1058],[224,-0.1722],[225,-0.3065],[226,-0.5024],[227,-0.7501],[228,-1.0432],[229,-1.3796],[230,-1.7414],[231,-2.1239],[232,-2.5003],[233,-2.8859],[234,-3.289],[235,-3.7196],[236,-4.1787],[237,-4.6524],[238,-5.1406],[240,-6.1383],[242,-7.1103],[243,-7.5836],[244,-8.0403],[245,-8.473],[246,-8.8732],[247,-9.2363],[248,-9.5343],[249,-9.7693],[250,-9.9176],[251,-9.9806],[252,-9.9792],[253,-9.9214],[254,-9.8208],[255,-9.6805],[256,-9.5156],[257,-9.3257],[261,-8.4573],[262,-8.279],[263,-8.1201],[264,-7.9959],[265,-7.9051],[266,-7.8443],[267,-7.8206],[271,-7.7895],[283,-7.4721],[284,-7.4993],[285,-7.613],[286,-7.8037],[287,-8.0413],[288,-8.3258],[290,-8.9448],[291,-9.2363],[292,-9.493],[293,-9.6999],[294,-9.8474],[295,-9.9742],[296,-10.0536],[297,-10.0907],[298,-10.0923],[299,-10.0696],[300,-10.0205],[301,-9.95],[303,-9.7443],[305,-9.4604],[309,-8.7649],[323,-4.3117],[324,-3.4371],[325,-3.0805],[326,-2.7387],[327,-2.4155],[328,-2.1057],[329,-1.8132],[330,-1.5327],[331,-1.2654],[332,-1.0114],[333,-0.7351],[334,-0.4007],[335,0.0085],[336,0.4682],[337,0.9722],[338,1.5186],[339,2.0984],[340,2.7007],[343,4.5803],[344,5.1965],[345,5.7862],[346,6.3423],[347,6.8525],[348,7.2913],[349,7.6628],[350,7.9697],[351,8.2364],[352,8.4664],[353,8.6489],[354,8.7942],[355,8.9023],[356,8.9823],[359,9.1164],[363,9.196],[370,9.1587],[374,9.0867],[375,9.0435],[376,9.0211],[379,8.8623],[382,8.6379],[391,7.7767],[396,7.2183],[398,6.9602],[400,6.665],[401,6.484],[402,6.265],[403,6.0198],[405,5.4756],[406,5.1901],[408,4.6003],[409,4.3351],[410,4.1482],[411,4.0503],[412,4.0319],[413,4.1458],[414,4.3656],[415,4.6501],[416,4.9799],[418,5.6639],[419,5.9786],[420,6.2358],[421,6.4087],[422,6.4781],[423,6.4535],[424,6.3643],[425,6.1971],[426,5.9501],[427,5.6313],[428,5.2144],[429,4.6984],[430,4.1038],[431,3.4618],[432,2.7769],[433,2.0674],[435,0.6307],[436,-0.0813],[437,-0.7708],[438,-1.4526],[439,-2.12],[440,-2.7617],[441,-3.3663],[442,-3.9127],[443,-4.402],[444,-4.817],[445,-5.1478],[446,-5.4115],[447,-5.6102],[448,-5.7475],[449,-5.8119],[450,-5.8239],[451,-5.7991],[452,-5.7535],[456,-5.4545],[457,-5.3875],[459,-5.2828],[462,-5.1634],[467,-4.995],[475,-4.715],[478,-4.6456],[480,-4.6271],[482,-4.6353],[485,-4.6954],[486,-4.7229],[487,-4.7848],[488,-4.9003],[489,-5.049],[490,-5.2282],[494,-6.0489],[495,-6.2447],[496,-6.4136],[497,-6.5397],[498,-6.6229],[499,-6.6775],[500,-6.7076],[501,-6.686],[502,-6.6401],[504,-6.5],[508,-6.1695],[509,-6.1048],[514,-5.6763],[516,-5.5239],[518,-5.392],[521,-5.2392],[523,-5.1798],[525,-5.1564],[531,-5.1896],[532,-5.1836],[533,-5.1536],[534,-5.0176],[535,-4.653],[536,-4.2348],[537,-3.7891],[538,-3.3346],[539,-2.8723],[540,-2.3949],[541,-1.9006],[542,-1.4337],[543,-1.0046],[544,-0.6509],[545,-0.3697],[546,-0.1862],[547,-0.106],[548,-0.1058],[549,-0.1722],[550,-0.3065],[551,-0.5024],[552,-0.7501],[553,-1.0432],[554,-1.3796],[555,-1.7414],[556,-2.1239],[557,-2.5003],[558,-2.8859],[559,-3.289],[560,-3.7196],[561,-4.1787],[562,-4.6524],[563,-5.1406],[565,-6.1383],[567,-7.1103],[568,-7.5836],[569,-8.0403],[570,-8.473],[571,-8.8732],[572,-9.2363],[573,-9.5343],[574,-9.7693],[575,-9.9176],[576,-9.9806],[577,-9.9792],[578,-9.9214],[579,-9.8208],[580,-9.6805],[581,-9.5156],[582,-9.3257],[586,-8.4573],[587,-8.279],[588,-8.1201],[589,-7.9959],[590,-7.9051],[591,-7.8443],[592,-7.8206],[596,-7.7895],[608,-7.4721],[609,-7.4993],[610,-7.613],[611,-7.8037],[612,-8.0413],[613,-8.3258],[615,-8.9448],[616,-9.2363],[617,-9.493],[618,-9.6999],[619,-9.8474],[620,-9.9742],[621,-10.0536],[622,-10.0907],[623,-10.0923],[624,-10.0696],[625,-10.0205],[626,-9.95],[628,-9.7443],[630,-9.4604],[634,-8.7649],[676,-4.3117],[677,-3.4371],[678,-3.0805],[679,-2.7387],[680,-2.4155],[681,-2.1057],[682,-1.8132],[683,-1.5327],[684,-1.2654],[685,-1.0114],[686,-0.7351],[687,-0.4007],[688,0.0085],[689,0.4682],[690,0.9722],[691,1.5186],[692,2.0984],[693,2.7007],[696,4.5803],[697,5.1965],[698,5.7862],[699,6.3423],[700,6.8525],[701,7.2913],[702,7.6628],[703,7.9697],[704,8.2364],[705,8.4664],[706,8.6489],[707,8.7942],[708,8.9023],[709,8.9823],[712,9.1164],[716,9.196],[723,9.1587],[727,9.0867],[728,9.0435],[729,9.0211],[732,8.8623],[735,8.6379],[744,7.7767],[749,7.2183],[751,6.9602],[753,6.665],[754,6.484],[755,6.265],[757,5.7531],[759,5.1901],[770,-7.613],[771,-7.8037],[772,-8.0413],[773,-8.3258],[775,-8.9448],[776,-9.2363],[777,-9.493],[778,-9.6999],[779,-9.8474],[780,-9.9742],[781,-10.0536],[782,-10.0907],[783,-10.0923],[784,-10.0696],[785,-10.0205],[786,-9.95],[788,-9.7443],[790,-9.4604],[798,-8.0636],[800,-7.7477],[803,-7.3434],[837,-4.3117],[887,-8.5858],[890,-8.0636],[892,-7.7477],[895,-7.3434],[929,-4.3117]]
       }
       head_ctrl.Rotation.Z {
        type=MovieSceneScriptingFloatChannel
        default=0
        thinned_from=751
        keys [[-2,-1.2578],[-1,-0.6105],[0,-0.3839],[1,-0.1608],[2,0.0568],[3,0.2686],[4,0.471],[5,0.6648],[6,0.851],[8,1.1988],[10,1.5297],[11,1.6877],[14,2.1206],[15,2.2559],[16,2.3826],[17,2.4969],[18,2.5887],[19,2.6684],[20,2.7273],[21,2.7742],[22,2.7933],[23,2.7916],[24,2.7742],[25,2.7435],[26,2.7033],[27,2.6509],[29,2.5284],[32,2.3183],[35,2.1252],[38,1.9451],[40,1.8342],[42,1.7384],[44,1.654],[45,1.6183],[46,1.5869],[47,1.56],[49,1.5209],[51,1.5031],[53,1.5073],[54,1.5165],[55,1.5309],[57,1.5706],[59,1.6218],[63,1.7493],[66,1.86],[67,1.8707],[68,1.8611],[69,1.8227],[70,1.7484],[71,1.6283],[72,1.4608],[73,1.2294],[74,0.9184],[75,0.4944],[76,-0.1238],[77,-0.9889],[78,-2.0326],[79,-3.2403],[80,-4.5462],[81,-5.9166],[82,-7.3344],[83,-8.7993],[84,-10.2425],[85,-11.6077],[86,-12.8335],[87,-13.874],[88,-14.6115],[89,-15.1795],[90,-15.6492],[91,-16.023],[92,-16.3106],[93,-16.5001],[94,-16.596],[95,-16.5822],[96,-16.4447],[97,-16.1859],[98,-15.8311],[99,-15.3622],[100,-14.8019],[101,-14.1524],[102,-13.4096],[104,-11.8296],[106,-10.3101],[107,-9.5825],[108,-8.8932],[109,-8.2422],[110,-7.6351],[111,-7.0779],[112,-6.5848],[113,-6.1333],[114,-5.7094],[115,-5.3274],[116,-4.9939],[117,-4.7095],[118,-4.4838],[119,-4.3182],[120,-4.2231],[121,-4.1824],[122,-4.1759],[123,-4.2076],[124,-4.289],[125,-4.4169],[126,-4.5994],[127,-4.8173],[129,-5.2975],[130,-5.523],[131,-5.7198],[132,-5.8937],[133,-6.0403],[134,-6.1566],[135,-6.2442],[136,-6.3039],[137,-6.3479],[139,-6.3673],[140,-6.3533],[142,-6.2862],[144,-6.17],[150,-5.7242],[152,-5.6026],[153,-5.5597],[154,-5.5304],[155,-5.5176],[156,-5.523],[157,-5.5482],[158,-5.5933],[159,-5.6646],[160,-5.7626],[161,-5.8912],[162,-6.0813],[163,-6.3647],[164,-6.7271],[165,-7.1446],[166,-7.6054],[169,-9.0778],[170,-9.5394],[171,-9.9542],[172,-10.3073],[173,-10.5927],[174,-10.8257],[176,-11.2262],[178,-11.5207],[180,-11.7265],[181,-11.7999],[182,-11.8465],[184,-11.8568],[190,-11.5147],[202,-11.0205],[203,-10.9302],[204,-10.8169],[205,-10.6671],[206,-10.4861],[207,-10.2556],[208,-9.9384],[209,-9.3679],[210,-8.2434],[211,-6.8972],[212,-5.3531],[213,-3.6502],[214,-1.8416],[215,0.0395],[216,1.9729],[217,3.9248],[218,5.8923],[219,7.8788],[222,13.912],[223,15.8402],[224,17.6879],[225,19.4175],[226,21.0126],[227,22.4473],[228,23.7064],[229,24.7562],[230,25.5815],[231,26.1584],[232,26.49],[233,26.6245],[234,26.5713],[235,26.3494],[236,25.9765],[237,25.4658],[238,24.8454],[239,24.1245],[240,23.3143],[241,22.4357],[243,20.526],[247,16.5258],[248,15.5548],[249,14.6219],[250,13.7426],[251,12.9224],[253,11.4337],[254,10.7128],[255,10.0176],[256,9.3517],[257,8.728],[258,8.1372],[259,7.5868],[260,7.0779],[261,6.6065],[262,6.1684],[263,5.7667],[264,5.409],[265,5.0942],[266,4.8228],[267,4.6119],[269,4.2446],[271,3.9158],[272,3.7668],[273,3.6316],[274,3.5067],[276,3.2866],[277,3.1845],[279,3.0082],[281,2.862],[282,2.7966],[283,2.7216],[284,2.5307],[285,2.0843],[286,1.4565],[287,0.6644],[288,-0.2326],[289,-1.2079],[290,-2.1868],[291,-3.1393],[292,-4.008],[293,-4.7403],[294,-5.2928],[295,-5.8068],[296,-6.1955],[297,-6.4896],[298,-6.7209],[299,-6.8867],[300,-7.0154],[301,-7.0923],[302,-7.1346],[303,-7.1343],[304,-7.0976],[305,-7.0203],[306,-6.9074],[307,-6.7751],[308,-6.6231],[309,-6.4573],[323,-1.2578],[324,-0.6105],[325,-0.3839],[326,-0.1608],[327,0.0568],[328,0.2686],[329,0.471],[330,0.6648],[331,0.851],[333,1.1988],[335,1.5297],[336,1.6877],[339,2.1206],[340,2.2559],[341,2.3826],[342,2.4969],[343,2.5887],[344,2.6684],[345,2.7273],[346,2.7742],[347,2.7933],[348,2.7916],[349,2.7742],[350,2.7435],[351,2.7033],[352,2.6509],[354,2.5284],[357,2.3183],[360,2.1252],[363,1.9451],[365,1.8342],[367,1.7384],[369,1.654],[370,1.6183],[371,1.5869],[372,1.56],[374,1.5209],[376,1.5031],[378,1.5073],[379,1.5165],[380,1.5309],[382,1.5706],[384,1.6218],[388,1.7493],[391,1.86],[392,1.8707],[393,1.8611],[394,1.8227],[395,1.7484],[396,1.6283],[397,1.4608],[398,1.2294],[399,0.9184],[400,0.4944],[401,-0.1238],[402,-0.9889],[403,-2.0326],[404,-3.2403],[405,-4.5462],[406,-5.9166],[407,-7.3344],[408,-8.7993],[409,-10.2425],[410,-11.6077],[411,-12.8335],[412,-13.874],[413,-14.6115],[414,-15.1795],[415,-15.6492],[416,-16.023],[417,-16.3106],[418,-16.5001],[419,-16.596],[420,-16.5822],[421,-16.4447],[422,-16.1859],[423,-15.8311],[424,-15.3622],[425,-14.8019],[426,-14.1524],[427,-13.4096],[429,-11.8296],[431,-10.3101],[432,-9.5825],[433,-8.8932],[434,-8.2422],[435,-7.6351],[436,-7.0779],[437,-6.5848],[438,-6.1333],[439,-5.7094],[440,-5.3274],[441,-4.9939],[442,-4.7095],[443,-4.4838],[444,-4.3182],[445,-4.2231],[446,-4.1824],[447,-4.1759],[448,-4.2076],[449,-4.289],[450,-4.4169],[451,-4.5994],[452,-4.8173],[454,-5.2975],[455,-5.523],[456,-5.7198],[457,-5.8937],[458,-6.0403],[459,-6.1566],[460,-6.2442],[461,-6.3039],[462,-6.3479],[464,-6.3673],[465,-6.3533],[467,-6.2862],[469,-6.17],[475,-5.7242],[477,-5.6026],[478,-5.5597],[479,-5.5304],[480,-5.5176],[481,-5.523],[482,-5.5482],[483,-5.5933],[484,-5.6646],[485,-5.7626],[486,-5.8912],[487,-6.0813],[488,-6.3647],[489,-6.7271],[490,-7.1446],[491,-7.6054],[494,-9.0778],[495,-9.5394],[496,-9.9542],[497,-10.3073],[498,-10.5927],[499,-10.8257],[501,-11.2262],[503,-11.5207],[505,-11.7265],[506,-11.7999],[507,-11.8465],[509,-11.8568],[515,-11.5147],[527,-11.0205],[528,-10.9302],[529,-10.8169],[530,-10.6671],[531,-10.4861],[532,-10.2556],[533,-9.9384],[534,-9.3679],[535,-8.2434],[536,-6.8972],[537,-5.3531],[538,-3.6502],[539,-1.8416],[540,0.0395],[541,1.9729],[542,3.9248],[543,5.8923],[544,7.8788],[547,13.912],[548,15.8402],[549,17.6879],[550,19.4175],[551,21.0126],[552,22.4473],[553,23.7064],[554,24.7562],[555,25.5815],[556,26.1584],[557,26.49],[558,26.6245],[559,26.5713],[560,26.3494],[561,25.9765],[562,25.4658],[563,24.8454],[564,24.1245],[565,23.3143],[566,22.4357],[568,20.526],[572,16.5258],[573,15.5548],[574,14.6219],[575,13.7426],[576,12.9224],[578,11.4337],[579,10.7128],[580,10.0176],[581,9.3517],[582,8.728],[583,8.1372],[584,7.5868],[585,7.0779],[586,6.6065],[587,6.1684],[588,5.7667],[589,5.409],[590,5.0942],[591,4.8228],[592,4.6119],[594,4.2446],[596,3.9158],[597,3.7668],[598,3.6316],[599,3.5067],[601,3.2866],[602,3.1845],[604,3.0082],[606,2.862],[607,2.7966],[608,2.7216],[609,2.5307],[610,2.0843],[611,1.4565],[612,0.6644],[613,-0.2326],[614,-1.2079],[615,-2.1868],[616,-3.1393],[617,-4.008],[618,-4.7403],[619,-5.2928],[620,-5.8068],[621,-6.1955],[622,-6.4896],[623,-6.7209],[624,-6.8867],[625,-7.0154],[626,-7.0923],[627,-7.1346],[628,-7.1343],[629,-7.0976],[630,-7.0203],[631,-6.9074],[632,-6.7751],[633,-6.6231],[634,-6.4573],[676,-1.2578],[677,-0.6105],[678,-0.3839],[679,-0.1608],[680,0.0568],[681,0.2686],[682,0.471],[683,0.6648],[684,0.851],[686,1.1988],[688,1.5297],[689,1.6877],[692,2.1206],[693,2.2559],[694,2.3826],[695,2.4969],[696,2.5887],[697,2.6684],[698,2.7273],[699,2.7742],[700,2.7933],[701,2.7916],[702,2.7742],[703,2.7435],[704,2.7033],[705,2.6509],[707,2.5284],[710,2.3183],[713,2.1252],[716,1.9451],[718,1.8342],[720,1.7384],[722,1.654],[723,1.6183],[724,1.5869],[725,1.56],[727,1.5209],[729,1.5031],[731,1.5073],[732,1.5165],[733,1.5309],[735,1.5706],[737,1.6218],[741,1.7493],[744,1.86],[745,1.8707],[746,1.8611],[747,1.8227],[748,1.7484],[749,1.6283],[750,1.4608],[751,1.2294],[752,0.9184],[753,0.4944],[754,-0.1238],[755,-0.9889],[757,-3.2403],[759,-5.9166],[770,2.0843],[771,1.4565],[772,0.6644],[773,-0.2326],[774,-1.2079],[775,-2.1868],[776,-3.1393],[777,-4.008],[778,-4.7403],[779,-5.2928],[780,-5.8068],[781,-6.1955],[782,-6.4896],[783,-6.7209],[784,-6.8867],[785,-7.0154],[786,-7.0923],[787,-7.1346],[788,-7.1343],[789,-7.0976],[790,-7.0203],[791,-6.9074],[792,-6.7751],[793,-6.6231],[795,-6.2867],[796,-6.1076],[797,-5.9085],[799,-5.4757],[801,-5.011],[803,-4.5686],[837,-1.2578],[887,-6.2867],[888,-6.1076],[889,-5.9085],[891,-5.4757],[893,-5.011],[895,-4.5686],[929,-1.2578]]
       }
       head_ctrl.Scale.X {type=MovieSceneScriptingFloatChannel, default=1, const [1,751]}
       head_ctrl.Scale.Y {type=MovieSceneScriptingFloatChannel, default=1, const [1,751]}
       head_ctrl.Scale.Z {type=MovieSceneScriptingFloatChannel, default=1, const [1,751]}
       clavicle_l_ctrl.Location.X {type=MovieSceneScriptingFloatChannel, default=0, thinned_from=750, keys [[-2,0],[929,0]]}
       clavicle_l_ctrl.Location.Y {type=MovieSceneScriptingFloatChannel, default=0, thinned_from=750, keys [[-2,0],[929,0]]}
       clavicle_l_ctrl.Location.Z {type=MovieSceneScriptingFloatChannel, default=0, thinned_from=750, keys [[-2,0],[929,0]]}
       clavicle_l_ctrl.Rotation.X {
        type=MovieSceneScriptingFloatChannel
        default=0
        thinned_from=750
        keys [[-2,5.9102],[759,5.9102],[868,5.9125],[879,5.9451],[895,5.9102],[929,5.9102]]
       }
       clavicle_l_ctrl.Rotation.Y {
        type=MovieSceneScriptingFloatChannel
        default=0
        thinned_from=715
        keys [[-2,0.2415],[759,0.2415],[782,4.2474],[818,2.2193],[851,1.4129],[868,12.8227],[879,10.5409],[893,0.2415],[929,0.2415]]
       }
       clavicle_l_ctrl.Rotation.Z {
        type=MovieSceneScriptingFloatChannel
        default=0
        thinned_from=750
        keys [[-2,11.4074],[759,11.4074],[770,11.5518],[803,11.4074],[866,11.4074],[868,11.4469],[879,12.0268],[893,11.4074],[929,11.4074]]
       }
       clavicle_l_ctrl.Scale.X {type=MovieSceneScriptingFloatChannel, default=1, const [1,750]}
       clavicle_l_ctrl.Scale.Y {type=MovieSceneScriptingFloatChannel, default=1, const [1,750]}
       clavicle_l_ctrl.Scale.Z {type=MovieSceneScriptingFloatChannel, default=1, const [1,750]}
       clavicle_r_ctrl.Location.X {type=MovieSceneScriptingFloatChannel, default=0, thinned_from=750, keys [[-2,0],[929,0]]}
       clavicle_r_ctrl.Location.Y {type=MovieSceneScriptingFloatChannel, default=0, thinned_from=750, keys [[-2,0],[929,0]]}
       clavicle_r_ctrl.Location.Z {type=MovieSceneScriptingFloatChannel, default=0, thinned_from=750, keys [[-2,0],[929,0]]}
       clavicle_r_ctrl.Rotation.X {
        type=MovieSceneScriptingFloatChannel
        default=0
        thinned_from=750
        keys [[-2,2.4237],[759,2.4237],[868,2.4246],[879,2.4374],[895,2.4237],[929,2.4237]]
       }
       clavicle_r_ctrl.Rotation.Y {
        type=MovieSceneScriptingFloatChannel
        default=0
        thinned_from=716
        keys [[-2,0.0791],[759,0.0791],[770,5.7575],[778,7.5609],[837,2.986],[856,2.986],[870,12.1522],[879,10.4051],[893,0.0791],[929,0.0791]]
       }
       clavicle_r_ctrl.Rotation.Z {
        type=MovieSceneScriptingFloatChannel
        default=0
        thinned_from=750
        keys [[-2,4.8999],[759,4.8999],[770,4.9592],[803,4.8999],[866,4.8999],[868,4.9161],[879,5.1542],[893,4.8999],[929,4.8999]]
       }
       clavicle_r_ctrl.Scale.X {type=MovieSceneScriptingFloatChannel, default=1, const [1,750]}
       clavicle_r_ctrl.Scale.Y {type=MovieSceneScriptingFloatChannel, default=1, const [1,750]}
       clavicle_r_ctrl.Scale.Z {type=MovieSceneScriptingFloatChannel, default=1, const [1,750]}
       neck_01_ctrl.Location.X {type=MovieSceneScriptingFloatChannel, default=0, thinned_from=753, keys [[-2,0],[929,0]]}
       neck_01_ctrl.Location.Y {type=MovieSceneScriptingFloatChannel, default=0, thinned_from=753, keys [[-2,0],[929,0]]}
       neck_01_ctrl.Location.Z {type=MovieSceneScriptingFloatChannel, default=0, thinned_from=753, keys [[-2,0],[929,0]]}
       neck_01_ctrl.Rotation.X {
        type=MovieSceneScriptingFloatChannel
        default=0
        thinned_from=686
        keys [[-2,10.7584],[90,9.8661],[94,9.6891],[106,8.8736],[110,8.7039],[124,8.3901],[125,8.3533],[126,8.3342],[127,8.2961],[128,8.2769],[129,8.2388],[130,8.2196],[143,7.7904],[144,7.771],[145,7.7264],[146,7.7071],[147,7.6688],[148,7.6495],[149,7.6112],[150,7.5919],[151,7.5568],[153,7.5183],[154,7.4832],[216,6.405],[218,6.4304],[220,6.4971],[223,6.6566],[231,7.1855],[233,7.2802],[235,7.3364],[237,7.3423],[287,7.0281],[290,7.0594],[293,7.1471],[297,7.3351],[309,8.1709],[329,10.7536],[415,9.8661],[419,9.6891],[431,8.8736],[435,8.7039],[449,8.3901],[450,8.3533],[451,8.3342],[452,8.2961],[453,8.2769],[454,8.2388],[455,8.2196],[468,7.7904],[469,7.771],[470,7.7264],[471,7.7071],[472,7.6688],[473,7.6495],[474,7.6112],[475,7.5919],[476,7.5568],[478,7.5183],[479,7.4832],[541,6.405],[543,6.4304],[545,6.4971],[548,6.6566],[556,7.1855],[558,7.2802],[560,7.3364],[562,7.3423],[612,7.0281],[615,7.0594],[618,7.1471],[622,7.3351],[634,8.1709],[676,10.7584],[736,10.0726],[781,7.2816],[831,6.4434],[855,8.2147],[871,15.2686],[902,11.2091],[923,10.7584],[929,10.7584]]
       }
       neck_01_ctrl.Rotation.Y {
        type=MovieSceneScriptingFloatChannel
        default=0
        thinned_from=556
        keys [[-2,-0.5797],[0,-0.5813],[2,-0.586],[4,-0.5938],[6,-0.6046],[8,-0.6185],[10,-0.6352],[12,-0.6549],[14,-0.6775],[17,-0.7166],[20,-0.762],[23,-0.8133],[26,-0.8706],[29,-0.9337],[33,-1.0264],[37,-1.1288],[42,-1.2698],[48,-1.4571],[55,-1.6991],[65,-2.0853],[80,-2.742],[92,-3.3244],[190,-8.7294],[197,-9.0804],[198,-9.0688],[199,-8.9314],[200,-8.6991],[201,-8.3715],[238,1.0298],[239,0.9136],[240,0.7672],[241,0.7473],[242,0.7217],[243,0.6924],[244,0.6581],[245,0.6202],[246,0.5788],[247,0.5337],[248,0.4866],[249,0.4366],[250,0.3846],[251,0.3312],[253,0.2204],[255,0.1069],[256,0.0495],[257,-0.0069],[258,-0.0623],[259,-0.1161],[260,-0.169],[261,-0.2194],[262,-0.2689],[263,-0.3144],[264,-0.3566],[265,-0.3956],[266,-0.4313],[267,-0.4621],[268,-0.489],[269,-0.511],[270,-0.5274],[271,-0.5389],[272,-0.5442],[273,-0.5421],[274,-0.5343],[281,-0.4501],[287,-0.3721],[323,-0.5797],[325,-0.5813],[327,-0.586],[329,-0.5938],[331,-0.6046],[333,-0.6185],[335,-0.6352],[337,-0.6549],[339,-0.6775],[342,-0.7166],[345,-0.762],[348,-0.8133],[351,-0.8706],[354,-0.9337],[358,-1.0264],[362,-1.1288],[367,-1.2698],[373,-1.4571],[380,-1.6991],[390,-2.0853],[405,-2.742],[417,-3.3244],[515,-8.7294],[522,-9.0804],[523,-9.0688],[524,-8.9314],[525,-8.6991],[526,-8.3715],[563,1.0298],[564,0.9136],[565,0.7672],[566,0.7473],[567,0.7217],[568,0.6924],[569,0.6581],[570,0.6202],[571,0.5788],[572,0.5337],[573,0.4866],[574,0.4366],[575,0.3846],[576,0.3312],[578,0.2204],[580,0.1069],[581,0.0495],[582,-0.0069],[583,-0.0623],[584,-0.1161],[585,-0.169],[586,-0.2194],[587,-0.2689],[588,-0.3144],[589,-0.3566],[590,-0.3956],[591,-0.4313],[592,-0.4621],[593,-0.489],[594,-0.511],[595,-0.5274],[596,-0.5389],[597,-0.5442],[598,-0.5421],[599,-0.5343],[606,-0.4501],[612,-0.3721],[676,2.6332],[682,2.6142],[688,2.5592],[694,2.4711],[701,2.3304],[708,2.1531],[715,1.9438],[722,1.7072],[726,1.5615],[770,-3.9149],[772,-3.9738],[801,-2.2022],[834,2.643],[837,2.643],[893,-2.2022],[926,-0.5699],[929,-0.5699]]
       }
       neck_01_ctrl.Rotation.Z {
        type=MovieSceneScriptingFloatChannel
        default=0
        thinned_from=179
        keys [[-2,0.9769],[19,0.9557],[35,0.9136],[83,0.7583],[97,-12.5292],[108,-17.5677],[143,-6.8523],[180,-3.462],[192,-3.5417],[203,-3.5482],[204,-3.5545],[210,-2.1129],[219,7.8289],[225,14.6346],[232,17.6407],[270,0.1243],[287,-6.8387],[288,-6.8508],[323,0.9769],[344,0.9557],[360,0.9136],[408,0.7583],[422,-12.5292],[433,-17.5677],[468,-6.8523],[505,-3.462],[517,-3.5417],[528,-3.5482],[529,-3.5545],[535,-2.1129],[544,7.8289],[550,14.6346],[557,17.6407],[595,0.1243],[612,-6.8387],[613,-6.8508],[676,0.9769],[693,0.9584],[713,0.8971],[748,0.7261],[772,-0.1689],[773,-0.1609],[801,1.3312],[802,1.3526],[803,1.3496],[834,0.9782],[837,0.9782],[893,1.3312],[894,1.3526],[895,1.3496],[926,0.9782],[929,0.9782]]
       }
       neck_01_ctrl.Scale.X {type=MovieSceneScriptingFloatChannel, default=1, const [1,753]}
       neck_01_ctrl.Scale.Y {type=MovieSceneScriptingFloatChannel, default=1, const [1,753]}
       neck_01_ctrl.Scale.Z {type=MovieSceneScriptingFloatChannel, default=1, const [1,753]}
       neck_02_ctrl.Location.X {
        type=MovieSceneScriptingFloatChannel
        default=0
        thinned_from=751
        keys [[-2,0.471],[0,0.4758],[3,0.4758],[4,0.4782],[17,0.478],[18,0.481],[23,0.481],[24,0.484],[28,0.484],[29,0.487],[32,0.4873],[33,0.4909],[37,0.4909],[38,0.4946],[42,0.4946],[43,0.4983],[62,0.4985],[63,0.4957],[66,0.4957],[67,0.4918],[72,0.4918],[73,0.4879],[76,0.4879],[77,0.4839],[79,0.4831],[84,0.474],[85,0.4703],[92,0.4574],[93,0.4574],[94,0.4556],[101,0.4545],[102,0.4555],[103,0.4555],[105,0.4574],[106,0.4574],[108,0.4593],[109,0.4593],[111,0.4605],[112,0.4619],[113,0.4619],[114,0.4634],[115,0.4634],[116,0.4648],[118,0.4648],[119,0.4662],[120,0.4662],[121,0.4677],[122,0.4677],[123,0.4691],[125,0.4691],[126,0.4705],[136,0.4719],[145,0.4723],[151,0.4716],[156,0.4708],[164,0.4695],[165,0.4733],[168,0.4733],[169,0.477],[171,0.477],[172,0.4808],[181,0.4808],[182,0.4862],[193,0.4862],[194,0.4891],[198,0.4891],[199,0.4919],[202,0.4919],[203,0.4947],[213,0.4952],[214,0.4916],[216,0.4916],[223,0.4662],[226,0.4522],[227,0.4522],[229,0.4428],[230,0.4428],[231,0.4382],[233,0.4382],[234,0.4335],[238,0.4335],[240,0.437],[241,0.437],[242,0.439],[243,0.439],[245,0.443],[247,0.443],[248,0.4449],[249,0.4449],[250,0.4469],[252,0.4469],[253,0.4489],[256,0.4489],[257,0.4513],[258,0.4513],[259,0.4537],[260,0.4537],[261,0.4561],[262,0.4561],[263,0.4585],[264,0.4585],[265,0.4609],[266,0.4609],[267,0.4633],[268,0.4633],[269,0.4656],[272,0.4664],[273,0.4714],[276,0.4714],[277,0.4764],[281,0.4764],[282,0.4813],[288,0.4813],[289,0.4798],[290,0.4798],[291,0.4782],[292,0.4782],[293,0.4766],[294,0.4766],[295,0.475],[296,0.475],[297,0.4734],[298,0.4734],[299,0.4719],[301,0.4719],[302,0.4703],[304,0.4703],[305,0.4686],[307,0.4686],[308,0.4668],[323,0.471],[325,0.4758],[328,0.4758],[329,0.4782],[342,0.478],[343,0.481],[348,0.481],[349,0.484],[353,0.484],[354,0.487],[357,0.4873],[358,0.4909],[362,0.4909],[363,0.4946],[367,0.4946],[368,0.4983],[387,0.4985],[388,0.4957],[391,0.4957],[392,0.4918],[397,0.4918],[398,0.4879],[401,0.4879],[402,0.4839],[404,0.4831],[409,0.474],[410,0.4703],[417,0.4574],[418,0.4574],[419,0.4556],[426,0.4545],[427,0.4555],[428,0.4555],[430,0.4574],[431,0.4574],[433,0.4593],[434,0.4593],[436,0.4605],[437,0.4619],[438,0.4619],[439,0.4634],[440,0.4634],[441,0.4648],[443,0.4648],[444,0.4662],[445,0.4662],[446,0.4677],[447,0.4677],[448,0.4691],[450,0.4691],[451,0.4705],[461,0.4719],[470,0.4723],[476,0.4716],[481,0.4708],[489,0.4695],[490,0.4733],[493,0.4733],[494,0.477],[496,0.477],[497,0.4808],[506,0.4808],[507,0.4862],[518,0.4862],[519,0.4891],[523,0.4891],[524,0.4919],[527,0.4919],[528,0.4947],[538,0.4952],[539,0.4916],[541,0.4916],[548,0.4662],[551,0.4522],[552,0.4522],[554,0.4428],[555,0.4428],[556,0.4382],[558,0.4382],[559,0.4335],[563,0.4335],[565,0.437],[566,0.437],[567,0.439],[568,0.439],[570,0.443],[572,0.443],[573,0.4449],[574,0.4449],[575,0.4469],[577,0.4469],[578,0.4489],[581,0.4489],[582,0.4513],[583,0.4513],[584,0.4537],[585,0.4537],[586,0.4561],[587,0.4561],[588,0.4585],[589,0.4585],[590,0.4609],[591,0.4609],[592,0.4633],[593,0.4633],[594,0.4656],[597,0.4664],[598,0.4714],[601,0.4714],[602,0.4764],[606,0.4764],[607,0.4813],[613,0.4813],[614,0.4798],[615,0.4798],[616,0.4782],[617,0.4782],[618,0.4766],[619,0.4766],[620,0.475],[621,0.475],[622,0.4734],[623,0.4734],[624,0.4719],[626,0.4719],[627,0.4703],[629,0.4703],[630,0.4686],[632,0.4686],[633,0.4668],[676,0.471],[678,0.4758],[681,0.4758],[682,0.4782],[695,0.478],[696,0.481],[701,0.481],[702,0.484],[706,0.484],[707,0.487],[710,0.4873],[711,0.4909],[715,0.4909],[716,0.4946],[720,0.4946],[721,0.4983],[740,0.4985],[741,0.4957],[744,0.4957],[745,0.4918],[750,0.4918],[751,0.4879],[754,0.4879],[755,0.4839],[757,0.4831],[759,0.4795],[773,0.4813],[774,0.4798],[775,0.4798],[776,0.4782],[777,0.4782],[778,0.4766],[779,0.4766],[780,0.475],[781,0.475],[782,0.4734],[783,0.4734],[784,0.4719],[786,0.4719],[787,0.4703],[789,0.4703],[790,0.4686],[792,0.4686],[793,0.4668],[796,0.4668],[797,0.4651],[803,0.4651],[837,0.471],[888,0.4668],[889,0.4651],[895,0.4651],[929,0.471]]
       }
       neck_02_ctrl.Location.Y {
        type=MovieSceneScriptingFloatChannel
        default=0
        thinned_from=751
        keys [[-2,0.4243],[-1,0.4238],[0,0.4295],[6,0.429],[7,0.4347],[12,0.4342],[13,0.4399],[17,0.4404],[18,0.4455],[21,0.4455],[22,0.4507],[26,0.4507],[27,0.4558],[30,0.4558],[31,0.4547],[32,0.4556],[36,0.4553],[62,0.4559],[63,0.4547],[64,0.4547],[65,0.4515],[73,0.4525],[74,0.4494],[76,0.4494],[77,0.4467],[78,0.4467],[82,0.442],[85,0.4394],[86,0.4396],[87,0.4389],[89,0.4406],[91,0.4406],[93,0.4382],[95,0.4342],[96,0.4343],[97,0.4249],[98,0.425],[99,0.4157],[100,0.4157],[101,0.4064],[102,0.4064],[104,0.3876],[105,0.3878],[106,0.3783],[108,0.3784],[109,0.369],[111,0.369],[112,0.3646],[113,0.3622],[114,0.3622],[116,0.3578],[118,0.3576],[119,0.3552],[121,0.3551],[122,0.3529],[123,0.3549],[124,0.3525],[125,0.3547],[126,0.3523],[127,0.3539],[129,0.3538],[130,0.3491],[136,0.3489],[137,0.3443],[140,0.3445],[141,0.3399],[143,0.34],[144,0.3357],[147,0.3359],[148,0.3317],[152,0.3319],[153,0.3276],[163,0.3279],[164,0.325],[166,0.325],[167,0.3244],[168,0.3215],[170,0.3209],[171,0.318],[178,0.3172],[179,0.3159],[184,0.3159],[185,0.3219],[191,0.3219],[192,0.3229],[195,0.3226],[196,0.3236],[206,0.3232],[207,0.3223],[210,0.3223],[211,0.3285],[213,0.3282],[214,0.3344],[215,0.3344],[216,0.3405],[217,0.3407],[219,0.3532],[220,0.3536],[222,0.3665],[223,0.3702],[224,0.3763],[225,0.3816],[226,0.3877],[227,0.393],[228,0.3938],[229,0.3991],[230,0.4052],[231,0.4052],[232,0.4061],[233,0.4113],[234,0.4122],[235,0.4122],[236,0.413],[237,0.4078],[238,0.4078],[241,0.3978],[242,0.3979],[245,0.388],[246,0.3881],[247,0.3847],[248,0.3847],[249,0.3814],[253,0.3813],[254,0.3846],[255,0.3834],[256,0.3853],[257,0.3853],[258,0.3873],[259,0.3873],[260,0.3892],[261,0.3892],[262,0.3912],[263,0.3912],[264,0.3931],[265,0.3931],[266,0.3953],[269,0.3948],[270,0.397],[275,0.3967],[276,0.394],[283,0.3935],[284,0.3907],[285,0.3907],[286,0.3882],[287,0.3866],[292,0.3705],[293,0.3707],[294,0.3676],[295,0.3676],[296,0.3645],[302,0.3648],[303,0.3664],[305,0.3664],[306,0.3721],[309,0.3719],[323,0.4243],[324,0.4238],[325,0.4295],[331,0.429],[332,0.4347],[337,0.4342],[338,0.4399],[342,0.4404],[343,0.4455],[346,0.4455],[347,0.4507],[351,0.4507],[352,0.4558],[355,0.4558],[356,0.4547],[357,0.4556],[361,0.4553],[387,0.4559],[388,0.4547],[389,0.4547],[390,0.4515],[398,0.4525],[399,0.4494],[401,0.4494],[402,0.4467],[403,0.4467],[407,0.442],[410,0.4394],[411,0.4396],[412,0.4389],[414,0.4406],[416,0.4406],[418,0.4382],[420,0.4342],[421,0.4343],[422,0.4249],[423,0.425],[424,0.4157],[425,0.4157],[426,0.4064],[427,0.4064],[429,0.3876],[430,0.3878],[431,0.3783],[433,0.3784],[434,0.369],[436,0.369],[437,0.3646],[438,0.3622],[439,0.3622],[441,0.3578],[443,0.3576],[444,0.3552],[446,0.3551],[447,0.3529],[448,0.3549],[449,0.3525],[450,0.3547],[451,0.3523],[452,0.3539],[454,0.3538],[455,0.3491],[461,0.3489],[462,0.3443],[465,0.3445],[466,0.3399],[468,0.34],[469,0.3357],[472,0.3359],[473,0.3317],[477,0.3319],[478,0.3276],[488,0.3279],[489,0.325],[491,0.325],[492,0.3244],[493,0.3215],[495,0.3209],[496,0.318],[503,0.3172],[504,0.3159],[509,0.3159],[510,0.3219],[516,0.3219],[517,0.3229],[520,0.3226],[521,0.3236],[531,0.3232],[532,0.3223],[535,0.3223],[536,0.3285],[538,0.3282],[539,0.3344],[540,0.3344],[541,0.3405],[542,0.3407],[544,0.3532],[545,0.3536],[547,0.3665],[548,0.3702],[549,0.3763],[550,0.3816],[551,0.3877],[552,0.393],[553,0.3938],[554,0.3991],[555,0.4052],[556,0.4052],[557,0.4061],[558,0.4113],[559,0.4122],[560,0.4122],[561,0.413],[562,0.4078],[563,0.4078],[566,0.3978],[567,0.3979],[570,0.388],[571,0.3881],[572,0.3847],[573,0.3847],[574,0.3814],[578,0.3813],[579,0.3846],[580,0.3834],[581,0.3853],[582,0.3853],[583,0.3873],[584,0.3873],[585,0.3892],[586,0.3892],[587,0.3912],[588,0.3912],[589,0.3931],[590,0.3931],[591,0.3953],[594,0.3948],[595,0.397],[600,0.3967],[601,0.394],[608,0.3935],[609,0.3907],[610,0.3907],[611,0.3882],[612,0.3866],[617,0.3705],[618,0.3707],[619,0.3676],[620,0.3676],[621,0.3645],[627,0.3648],[628,0.3664],[630,0.3664],[631,0.3721],[634,0.3719],[676,0.4243],[677,0.4238],[678,0.4295],[684,0.429],[685,0.4347],[690,0.4342],[691,0.4399],[695,0.4404],[696,0.4455],[699,0.4455],[700,0.4507],[704,0.4507],[705,0.4558],[708,0.4558],[709,0.4547],[710,0.4556],[714,0.4553],[740,0.4559],[741,0.4547],[742,0.4547],[743,0.4515],[751,0.4525],[752,0.4494],[754,0.4494],[755,0.4467],[759,0.4435],[770,0.3907],[771,0.3882],[772,0.3866],[777,0.3705],[778,0.3707],[779,0.3676],[780,0.3676],[781,0.3645],[787,0.3648],[788,0.3664],[790,0.3664],[791,0.3721],[795,0.3719],[796,0.3776],[800,0.3774],[801,0.3831],[803,0.3829],[837,0.4243],[887,0.3719],[888,0.3776],[892,0.3774],[893,0.3831],[895,0.3829],[929,0.4243]]
       }
       neck_02_ctrl.Location.Z {
        type=MovieSceneScriptingFloatChannel
        default=0
        thinned_from=751
        keys [[-2,-0.1907],[-1,-0.192],[0,-0.1939],[3,-0.1939],[4,-0.1953],[6,-0.1953],[7,-0.1972],[11,-0.1972],[12,-0.1986],[13,-0.2005],[17,-0.2008],[18,-0.2044],[21,-0.2044],[22,-0.208],[26,-0.208],[27,-0.2116],[30,-0.2116],[31,-0.2112],[32,-0.2115],[33,-0.2125],[36,-0.2125],[37,-0.2138],[42,-0.2138],[43,-0.2148],[62,-0.215],[63,-0.2131],[64,-0.2131],[65,-0.2121],[66,-0.2121],[67,-0.2107],[72,-0.2107],[74,-0.2082],[76,-0.2082],[77,-0.2057],[78,-0.2057],[80,-0.204],[84,-0.1997],[87,-0.1979],[91,-0.1981],[93,-0.1973],[94,-0.1961],[96,-0.1953],[97,-0.1922],[98,-0.1918],[99,-0.1883],[100,-0.1883],[101,-0.1848],[102,-0.1848],[104,-0.1782],[105,-0.1778],[106,-0.1746],[108,-0.1743],[109,-0.1711],[110,-0.1715],[111,-0.1713],[112,-0.1699],[114,-0.1697],[116,-0.1682],[117,-0.1687],[120,-0.1685],[121,-0.169],[122,-0.1683],[123,-0.1695],[124,-0.1693],[125,-0.1701],[126,-0.1698],[128,-0.1709],[129,-0.1709],[130,-0.1698],[132,-0.1698],[133,-0.1703],[136,-0.1703],[137,-0.1688],[138,-0.1688],[139,-0.1683],[140,-0.1683],[141,-0.1668],[143,-0.1666],[144,-0.1652],[145,-0.1647],[147,-0.1647],[148,-0.1633],[149,-0.1633],[150,-0.1628],[152,-0.1628],[153,-0.1614],[154,-0.1614],[155,-0.1609],[163,-0.1606],[164,-0.1597],[166,-0.1597],[167,-0.1614],[168,-0.1604],[169,-0.1604],[170,-0.1621],[171,-0.1611],[172,-0.1611],[173,-0.1628],[174,-0.1628],[175,-0.1632],[178,-0.1632],[179,-0.1674],[184,-0.1674],[185,-0.1694],[191,-0.1693],[193,-0.1702],[195,-0.1702],[197,-0.1711],[202,-0.1711],[203,-0.1717],[210,-0.1717],[211,-0.1738],[212,-0.1744],[213,-0.1744],[214,-0.1765],[215,-0.1765],[216,-0.1785],[217,-0.1779],[218,-0.1799],[219,-0.1813],[220,-0.18],[221,-0.1814],[222,-0.1822],[223,-0.1822],[224,-0.1814],[225,-0.1832],[226,-0.1824],[227,-0.1841],[228,-0.1816],[229,-0.1834],[230,-0.1826],[231,-0.1826],[232,-0.18],[233,-0.1818],[234,-0.1793],[235,-0.1793],[236,-0.1767],[237,-0.175],[238,-0.175],[241,-0.1706],[242,-0.1703],[245,-0.1662],[246,-0.1659],[247,-0.1648],[248,-0.1648],[249,-0.1633],[250,-0.1633],[251,-0.1637],[253,-0.1637],[254,-0.1651],[255,-0.1649],[256,-0.1663],[257,-0.1663],[258,-0.1678],[259,-0.1678],[260,-0.1693],[261,-0.1693],[262,-0.1708],[263,-0.1708],[264,-0.1723],[265,-0.1723],[267,-0.1737],[268,-0.1737],[270,-0.1752],[272,-0.1753],[273,-0.1761],[275,-0.1761],[276,-0.1751],[277,-0.1759],[280,-0.1759],[281,-0.1766],[283,-0.1766],[284,-0.1757],[285,-0.1757],[286,-0.174],[287,-0.1734],[288,-0.1715],[289,-0.1704],[291,-0.1666],[292,-0.1655],[293,-0.1647],[294,-0.1628],[295,-0.1628],[296,-0.1609],[299,-0.1609],[300,-0.1601],[302,-0.1601],[303,-0.1606],[305,-0.1606],[306,-0.1625],[307,-0.163],[309,-0.163],[323,-0.1907],[324,-0.192],[325,-0.1939],[328,-0.1939],[329,-0.1953],[331,-0.1953],[332,-0.1972],[336,-0.1972],[337,-0.1986],[338,-0.2005],[342,-0.2008],[343,-0.2044],[346,-0.2044],[347,-0.208],[351,-0.208],[352,-0.2116],[355,-0.2116],[356,-0.2112],[357,-0.2115],[358,-0.2125],[361,-0.2125],[362,-0.2138],[367,-0.2138],[368,-0.2148],[387,-0.215],[388,-0.2131],[389,-0.2131],[390,-0.2121],[391,-0.2121],[392,-0.2107],[397,-0.2107],[399,-0.2082],[401,-0.2082],[402,-0.2057],[403,-0.2057],[405,-0.204],[409,-0.1997],[412,-0.1979],[416,-0.1981],[418,-0.1973],[419,-0.1961],[421,-0.1953],[422,-0.1922],[423,-0.1918],[424,-0.1883],[425,-0.1883],[426,-0.1848],[427,-0.1848],[429,-0.1782],[430,-0.1778],[431,-0.1746],[433,-0.1743],[434,-0.1711],[435,-0.1715],[436,-0.1713],[437,-0.1699],[439,-0.1697],[441,-0.1682],[442,-0.1687],[445,-0.1685],[446,-0.169],[447,-0.1683],[448,-0.1695],[449,-0.1693],[450,-0.1701],[451,-0.1698],[453,-0.1709],[454,-0.1709],[455,-0.1698],[457,-0.1698],[458,-0.1703],[461,-0.1703],[462,-0.1688],[463,-0.1688],[464,-0.1683],[465,-0.1683],[466,-0.1668],[468,-0.1666],[469,-0.1652],[470,-0.1647],[472,-0.1647],[473,-0.1633],[474,-0.1633],[475,-0.1628],[477,-0.1628],[478,-0.1614],[479,-0.1614],[480,-0.1609],[488,-0.1606],[489,-0.1597],[491,-0.1597],[492,-0.1614],[493,-0.1604],[494,-0.1604],[495,-0.1621],[496,-0.1611],[497,-0.1611],[498,-0.1628],[499,-0.1628],[500,-0.1632],[503,-0.1632],[504,-0.1674],[509,-0.1674],[510,-0.1694],[516,-0.1693],[518,-0.1702],[520,-0.1702],[522,-0.1711],[527,-0.1711],[528,-0.1717],[535,-0.1717],[536,-0.1738],[537,-0.1744],[538,-0.1744],[539,-0.1765],[540,-0.1765],[541,-0.1785],[542,-0.1779],[543,-0.1799],[544,-0.1813],[545,-0.18],[546,-0.1814],[547,-0.1822],[548,-0.1822],[549,-0.1814],[550,-0.1832],[551,-0.1824],[552,-0.1841],[553,-0.1816],[554,-0.1834],[555,-0.1826],[556,-0.1826],[557,-0.18],[558,-0.1818],[559,-0.1793],[560,-0.1793],[561,-0.1767],[562,-0.175],[563,-0.175],[566,-0.1706],[567,-0.1703],[570,-0.1662],[571,-0.1659],[572,-0.1648],[573,-0.1648],[574,-0.1633],[575,-0.1633],[576,-0.1637],[578,-0.1637],[579,-0.1651],[580,-0.1649],[581,-0.1663],[582,-0.1663],[583,-0.1678],[584,-0.1678],[585,-0.1693],[586,-0.1693],[587,-0.1708],[588,-0.1708],[589,-0.1723],[590,-0.1723],[592,-0.1737],[593,-0.1737],[595,-0.1752],[597,-0.1753],[598,-0.1761],[600,-0.1761],[601,-0.1751],[602,-0.1759],[605,-0.1759],[606,-0.1766],[608,-0.1766],[609,-0.1757],[610,-0.1757],[611,-0.174],[612,-0.1734],[613,-0.1715],[614,-0.1704],[616,-0.1666],[617,-0.1655],[618,-0.1647],[619,-0.1628],[620,-0.1628],[621,-0.1609],[624,-0.1609],[625,-0.1601],[627,-0.1601],[628,-0.1606],[630,-0.1606],[631,-0.1625],[632,-0.163],[634,-0.163],[676,-0.1907],[677,-0.192],[678,-0.1939],[681,-0.1939],[682,-0.1953],[684,-0.1953],[685,-0.1972],[689,-0.1972],[690,-0.1986],[691,-0.2005],[695,-0.2008],[696,-0.2044],[699,-0.2044],[700,-0.208],[704,-0.208],[705,-0.2116],[708,-0.2116],[709,-0.2112],[710,-0.2115],[711,-0.2125],[714,-0.2125],[715,-0.2138],[720,-0.2138],[721,-0.2148],[740,-0.215],[741,-0.2131],[742,-0.2131],[743,-0.2121],[744,-0.2121],[745,-0.2107],[750,-0.2107],[752,-0.2082],[754,-0.2082],[755,-0.2057],[757,-0.2049],[759,-0.2027],[770,-0.1757],[771,-0.174],[772,-0.1734],[773,-0.1715],[774,-0.1704],[776,-0.1666],[777,-0.1655],[778,-0.1647],[779,-0.1628],[780,-0.1628],[781,-0.1609],[784,-0.1609],[785,-0.1601],[787,-0.1601],[788,-0.1606],[790,-0.1606],[791,-0.1625],[792,-0.163],[795,-0.163],[796,-0.1649],[797,-0.1649],[798,-0.1655],[800,-0.1655],[801,-0.1674],[802,-0.168],[803,-0.168],[837,-0.1907],[887,-0.163],[888,-0.1649],[889,-0.1649],[890,-0.1655],[892,-0.1655],[893,-0.1674],[894,-0.168],[895,-0.168],[929,-0.1907]]
       }
       neck_02_ctrl.Rotation.X {
        type=MovieSceneScriptingFloatChannel
        default=0
        thinned_from=751
        keys [[-2,18.1252],[88,18.129],[91,18.0295],[94,17.7748],[99,17.105],[107,15.871],[110,15.512],[115,15.012],[120,14.5805],[128,13.9664],[135,13.5426],[144,13.2657],[150,13.162],[163,13.1112],[165,13.0439],[167,12.8465],[168,12.7281],[171,12.2538],[176,11.3768],[178,11.0798],[180,10.8555],[181,10.7797],[182,10.7269],[215,10.8067],[216,10.86],[217,10.9883],[218,11.1845],[219,11.4449],[220,11.7626],[221,12.1232],[222,12.5267],[223,12.9569],[225,13.8866],[229,15.7993],[230,16.2389],[231,16.6443],[232,17.0055],[233,17.3275],[234,17.5854],[235,17.7791],[236,17.8987],[237,17.939],[309,18.0303],[413,18.129],[416,18.0295],[419,17.7748],[424,17.105],[432,15.871],[435,15.512],[440,15.012],[445,14.5805],[453,13.9664],[460,13.5426],[469,13.2657],[475,13.162],[488,13.1112],[490,13.0439],[492,12.8465],[493,12.7281],[496,12.2538],[501,11.3768],[503,11.0798],[505,10.8555],[506,10.7797],[507,10.7269],[540,10.8067],[541,10.86],[542,10.9883],[543,11.1845],[544,11.4449],[545,11.7626],[546,12.1232],[547,12.5267],[548,12.9569],[550,13.8866],[554,15.7993],[555,16.2389],[556,16.6443],[557,17.0055],[558,17.3275],[559,17.5854],[560,17.7791],[561,17.8987],[562,17.939],[634,18.0303],[755,18.2705],[803,18.0336],[837,18.1252],[895,18.0336],[929,18.1252]]
       }
       neck_02_ctrl.Rotation.Y {
        type=MovieSceneScriptingFloatChannel
        default=0
        thinned_from=751
        keys [[-2,-9.4595],[218,-9.2001],[233,-9.4043],[309,-9.463],[543,-9.2001],[558,-9.4043],[634,-9.463],[759,-9.4345],[803,-9.4688],[929,-9.4595]]
       }
       neck_02_ctrl.Rotation.Z {
        type=MovieSceneScriptingFloatChannel
        default=0
        thinned_from=751
        keys [[-2,3.1832],[-1,3.1729],[25,3.1954],[56,3.1193],[84,3.2079],[89,3.2537],[96,3.3375],[104,3.364],[151,3.3295],[163,3.3375],[173,3.2839],[209,3.1992],[215,3.216],[217,3.2353],[220,3.2899],[227,3.4674],[230,3.5212],[232,3.5403],[235,3.5434],[239,3.5095],[249,3.3705],[291,2.999],[296,2.9896],[302,3.0182],[309,3.0765],[323,3.1832],[324,3.1729],[350,3.1954],[381,3.1193],[409,3.2079],[414,3.2537],[421,3.3375],[429,3.364],[476,3.3295],[488,3.3375],[498,3.2839],[534,3.1992],[540,3.216],[542,3.2353],[545,3.2899],[552,3.4674],[555,3.5212],[557,3.5403],[560,3.5434],[564,3.5095],[574,3.3705],[616,2.999],[621,2.9896],[627,3.0182],[634,3.0765],[676,3.1832],[677,3.1729],[703,3.1954],[734,3.1193],[755,3.1711],[759,3.1907],[770,3.0512],[777,2.9938],[782,2.9919],[787,3.0182],[798,3.1089],[803,3.1315],[837,3.1832],[887,3.0869],[895,3.1315],[929,3.1832]]
       }
       neck_02_ctrl.Scale.X {type=MovieSceneScriptingFloatChannel, default=1, const [1,751]}
       neck_02_ctrl.Scale.Y {type=MovieSceneScriptingFloatChannel, default=1, const [1,751]}
       neck_02_ctrl.Scale.Z {type=MovieSceneScriptingFloatChannel, default=1, const [1,751]}
       head_ik_ctrl.Location.X {
        type=MovieSceneScriptingFloatChannel
        default=0
        thinned_from=751
        keys [[-2,1.2969],[0,1.306],[3,1.3124],[6,1.3257],[17,1.3665],[18,1.3737],[23,1.3898],[24,1.3951],[25,1.3973],[26,1.403],[32,1.4137],[33,1.4178],[37,1.4214],[38,1.4262],[42,1.4293],[43,1.4333],[58,1.4379],[59,1.4346],[62,1.4351],[66,1.431],[67,1.4265],[72,1.4256],[73,1.4216],[76,1.4217],[77,1.418],[85,1.4181],[93,1.4467],[96,1.4406],[98,1.4279],[99,1.4179],[100,1.4113],[101,1.4004],[102,1.3942],[103,1.382],[106,1.3597],[110,1.3421],[118,1.3277],[129,1.3169],[133,1.3084],[134,1.3021],[157,1.2796],[164,1.2824],[165,1.2848],[168,1.2822],[169,1.2837],[171,1.2787],[173,1.2762],[177,1.2636],[178,1.2546],[181,1.2461],[182,1.2498],[193,1.2525],[194,1.2559],[197,1.2585],[202,1.2685],[203,1.2728],[210,1.2858],[211,1.2824],[214,1.2892],[215,1.2859],[216,1.2905],[217,1.2816],[218,1.2471],[219,1.2003],[220,1.1351],[221,1.0565],[222,0.9625],[223,0.8579],[224,0.7423],[225,0.6164],[226,0.4872],[227,0.3599],[228,0.2279],[229,0.1022],[230,-0.0171],[231,-0.1301],[232,-0.223],[233,-0.3035],[234,-0.3688],[235,-0.4093],[236,-0.4257],[237,-0.4255],[238,-0.416],[239,-0.3992],[241,-0.3475],[242,-0.3135],[243,-0.2767],[245,-0.183],[246,-0.1342],[247,-0.081],[248,-0.024],[250,0.0992],[251,0.1612],[252,0.2254],[253,0.2914],[254,0.3566],[256,0.4888],[257,0.5567],[259,0.687],[261,0.811],[263,0.9258],[265,1.0263],[266,1.0699],[267,1.111],[268,1.1459],[269,1.1784],[270,1.2038],[271,1.2255],[273,1.2544],[274,1.2577],[276,1.258],[277,1.2632],[281,1.2636],[282,1.2687],[290,1.2716],[291,1.2703],[292,1.2725],[295,1.2703],[296,1.273],[303,1.2809],[309,1.295],[323,1.2969],[325,1.306],[328,1.3124],[331,1.3257],[342,1.3665],[343,1.3737],[348,1.3898],[349,1.3951],[350,1.3973],[351,1.403],[357,1.4137],[358,1.4178],[362,1.4214],[363,1.4262],[367,1.4293],[368,1.4333],[383,1.4379],[384,1.4346],[387,1.4351],[391,1.431],[392,1.4265],[397,1.4256],[398,1.4216],[401,1.4217],[402,1.418],[410,1.4181],[418,1.4467],[421,1.4406],[423,1.4279],[424,1.4179],[425,1.4113],[426,1.4004],[427,1.3942],[428,1.382],[431,1.3597],[435,1.3421],[443,1.3277],[454,1.3169],[458,1.3084],[459,1.3021],[482,1.2796],[489,1.2824],[490,1.2848],[493,1.2822],[494,1.2837],[496,1.2787],[498,1.2762],[502,1.2636],[503,1.2546],[506,1.2461],[507,1.2498],[518,1.2525],[519,1.2559],[522,1.2585],[527,1.2685],[528,1.2728],[535,1.2858],[536,1.2824],[539,1.2892],[540,1.2859],[541,1.2905],[542,1.2816],[543,1.2471],[544,1.2003],[545,1.1351],[546,1.0565],[547,0.9625],[548,0.8579],[549,0.7423],[550,0.6164],[551,0.4872],[552,0.3599],[553,0.2279],[554,0.1022],[555,-0.0171],[556,-0.1301],[557,-0.223],[558,-0.3035],[559,-0.3688],[560,-0.4093],[561,-0.4257],[562,-0.4255],[563,-0.416],[564,-0.3992],[566,-0.3475],[567,-0.3135],[568,-0.2767],[570,-0.183],[571,-0.1342],[572,-0.081],[573,-0.024],[575,0.0992],[576,0.1612],[577,0.2254],[578,0.2914],[579,0.3566],[581,0.4888],[582,0.5567],[584,0.687],[586,0.811],[588,0.9258],[590,1.0263],[591,1.0699],[592,1.111],[593,1.1459],[594,1.1784],[595,1.2038],[596,1.2255],[598,1.2544],[599,1.2577],[601,1.258],[602,1.2632],[606,1.2636],[607,1.2687],[615,1.2716],[616,1.2703],[617,1.2725],[620,1.2703],[621,1.273],[628,1.2809],[634,1.295],[676,1.2969],[678,1.306],[681,1.3124],[684,1.3257],[695,1.3665],[696,1.3737],[701,1.3898],[702,1.3951],[703,1.3973],[704,1.403],[710,1.4137],[711,1.4178],[715,1.4214],[716,1.4262],[720,1.4293],[721,1.4333],[736,1.4379],[737,1.4346],[740,1.4351],[744,1.431],[745,1.4265],[750,1.4256],[751,1.4216],[754,1.4217],[755,1.418],[759,1.4167],[770,1.2693],[775,1.2716],[776,1.2703],[777,1.2725],[780,1.2703],[781,1.273],[788,1.2809],[803,1.3159],[837,1.2969],[887,1.2968],[895,1.3159],[929,1.2969]]
       }
       head_ik_ctrl.Location.Y {
        type=MovieSceneScriptingFloatChannel
        default=0
        thinned_from=751
        keys [[-2,0.8043],[-1,0.8008],[0,0.8048],[4,0.7899],[5,0.7799],[6,0.7745],[7,0.774],[11,0.7486],[12,0.7378],[13,0.7373],[17,0.7122],[18,0.7115],[19,0.705],[20,0.6915],[21,0.6854],[22,0.6859],[25,0.6706],[26,0.667],[27,0.6696],[28,0.6662],[31,0.6613],[32,0.6617],[36,0.6587],[37,0.6597],[38,0.6582],[47,0.6552],[61,0.6486],[62,0.6491],[63,0.6463],[64,0.6457],[65,0.642],[67,0.6411],[68,0.6422],[72,0.6397],[73,0.6411],[74,0.6368],[76,0.6381],[77,0.6359],[78,0.6378],[79,0.6372],[81,0.6382],[83,0.6355],[84,0.6304],[85,0.6235],[86,0.6144],[87,0.602],[88,0.5876],[89,0.5698],[90,0.5472],[91,0.5191],[92,0.4862],[94,0.4108],[96,0.3321],[97,0.2922],[98,0.2558],[99,0.2184],[101,0.1528],[102,0.1238],[103,0.0887],[104,0.063],[105,0.0434],[106,0.0225],[107,0.0139],[108,0.0005],[109,-0.0127],[110,-0.0155],[111,-0.0166],[112,-0.0234],[113,-0.0281],[115,-0.0313],[116,-0.0346],[117,-0.0364],[118,-0.0362],[119,-0.0422],[120,-0.0493],[121,-0.0546],[122,-0.0677],[123,-0.0746],[125,-0.1009],[126,-0.1172],[127,-0.1304],[128,-0.1447],[129,-0.1597],[130,-0.1772],[131,-0.1915],[132,-0.2038],[133,-0.2155],[134,-0.2257],[135,-0.2478],[136,-0.256],[137,-0.269],[138,-0.2761],[140,-0.2882],[141,-0.2979],[142,-0.3026],[143,-0.3083],[144,-0.3167],[145,-0.3219],[146,-0.3252],[147,-0.3293],[148,-0.3367],[150,-0.3433],[152,-0.3486],[153,-0.3536],[154,-0.3557],[156,-0.3576],[158,-0.3573],[159,-0.3565],[163,-0.3557],[164,-0.3583],[165,-0.3647],[167,-0.3802],[168,-0.394],[169,-0.4098],[170,-0.4243],[171,-0.443],[172,-0.4591],[177,-0.5461],[179,-0.5771],[181,-0.6002],[182,-0.608],[183,-0.6105],[184,-0.6106],[185,-0.6044],[190,-0.6018],[192,-0.5984],[194,-0.5977],[198,-0.5902],[199,-0.5901],[203,-0.5795],[207,-0.5647],[209,-0.5552],[210,-0.5537],[211,-0.5407],[212,-0.5343],[213,-0.5306],[214,-0.5127],[215,-0.5057],[216,-0.4832],[217,-0.4528],[218,-0.4002],[219,-0.3387],[220,-0.2698],[221,-0.1895],[222,-0.1036],[223,-0.0221],[224,0.063],[225,0.1449],[226,0.2229],[227,0.2944],[228,0.3534],[229,0.4116],[230,0.4656],[231,0.5031],[233,0.5674],[234,0.5879],[235,0.6041],[236,0.6159],[237,0.6121],[238,0.614],[239,0.6118],[240,0.6119],[241,0.6149],[242,0.6207],[243,0.6232],[244,0.6286],[246,0.6446],[247,0.6503],[248,0.6598],[249,0.6663],[250,0.6781],[253,0.7036],[254,0.7146],[255,0.7207],[256,0.7292],[257,0.7352],[258,0.7428],[267,0.7896],[270,0.803],[273,0.812],[274,0.8114],[275,0.8132],[276,0.8121],[277,0.8141],[278,0.8121],[283,0.8186],[284,0.8143],[285,0.816],[287,0.8147],[288,0.8112],[289,0.813],[291,0.805],[292,0.8063],[293,0.8053],[294,0.8007],[295,0.7989],[296,0.7993],[300,0.7874],[302,0.7789],[303,0.7758],[305,0.7667],[306,0.767],[307,0.7582],[308,0.7523],[309,0.7483],[323,0.8043],[324,0.8008],[325,0.8048],[329,0.7899],[330,0.7799],[331,0.7745],[332,0.774],[336,0.7486],[337,0.7378],[338,0.7373],[342,0.7122],[343,0.7115],[344,0.705],[345,0.6915],[346,0.6854],[347,0.6859],[350,0.6706],[351,0.667],[352,0.6696],[353,0.6662],[356,0.6613],[357,0.6617],[361,0.6587],[362,0.6597],[363,0.6582],[372,0.6552],[386,0.6486],[387,0.6491],[388,0.6463],[389,0.6457],[390,0.642],[392,0.6411],[393,0.6422],[397,0.6397],[398,0.6411],[399,0.6368],[401,0.6381],[402,0.6359],[403,0.6378],[404,0.6372],[406,0.6382],[408,0.6355],[409,0.6304],[410,0.6235],[411,0.6144],[412,0.602],[413,0.5876],[414,0.5698],[415,0.5472],[416,0.5191],[417,0.4862],[419,0.4108],[421,0.3321],[422,0.2922],[423,0.2558],[424,0.2184],[426,0.1528],[427,0.1238],[428,0.0887],[429,0.063],[430,0.0434],[431,0.0225],[432,0.0139],[433,0.0005],[434,-0.0127],[435,-0.0155],[436,-0.0166],[437,-0.0234],[438,-0.0281],[440,-0.0313],[441,-0.0346],[442,-0.0364],[443,-0.0362],[444,-0.0422],[445,-0.0493],[446,-0.0546],[447,-0.0677],[448,-0.0746],[450,-0.1009],[451,-0.1172],[452,-0.1304],[453,-0.1447],[454,-0.1597],[455,-0.1772],[456,-0.1915],[457,-0.2038],[458,-0.2155],[459,-0.2257],[460,-0.2478],[461,-0.256],[462,-0.269],[463,-0.2761],[465,-0.2882],[466,-0.2979],[467,-0.3026],[468,-0.3083],[469,-0.3167],[470,-0.3219],[471,-0.3252],[472,-0.3293],[473,-0.3367],[475,-0.3433],[477,-0.3486],[478,-0.3536],[479,-0.3557],[481,-0.3576],[483,-0.3573],[484,-0.3565],[488,-0.3557],[489,-0.3583],[490,-0.3647],[492,-0.3802],[493,-0.394],[494,-0.4098],[495,-0.4243],[496,-0.443],[497,-0.4591],[502,-0.5461],[504,-0.5771],[506,-0.6002],[507,-0.608],[508,-0.6105],[509,-0.6106],[510,-0.6044],[515,-0.6018],[517,-0.5984],[519,-0.5977],[523,-0.5902],[524,-0.5901],[528,-0.5795],[532,-0.5647],[534,-0.5552],[535,-0.5537],[536,-0.5407],[537,-0.5343],[538,-0.5306],[539,-0.5127],[540,-0.5057],[541,-0.4832],[542,-0.4528],[543,-0.4002],[544,-0.3387],[545,-0.2698],[546,-0.1895],[547,-0.1036],[548,-0.0221],[549,0.063],[550,0.1449],[551,0.2229],[552,0.2944],[553,0.3534],[554,0.4116],[555,0.4656],[556,0.5031],[558,0.5674],[559,0.5879],[560,0.6041],[561,0.6159],[562,0.6121],[563,0.614],[564,0.6118],[565,0.6119],[566,0.6149],[567,0.6207],[568,0.6232],[569,0.6286],[571,0.6446],[572,0.6503],[573,0.6598],[574,0.6663],[575,0.6781],[578,0.7036],[579,0.7146],[580,0.7207],[581,0.7292],[582,0.7352],[583,0.7428],[592,0.7896],[595,0.803],[598,0.812],[599,0.8114],[600,0.8132],[601,0.8121],[602,0.8141],[603,0.8121],[608,0.8186],[609,0.8143],[610,0.816],[612,0.8147],[613,0.8112],[614,0.813],[616,0.805],[617,0.8063],[618,0.8053],[619,0.8007],[620,0.7989],[621,0.7993],[625,0.7874],[627,0.7789],[628,0.7758],[630,0.7667],[631,0.767],[632,0.7582],[633,0.7523],[634,0.7483],[676,0.8043],[677,0.8008],[678,0.8048],[682,0.7899],[683,0.7799],[684,0.7745],[685,0.774],[689,0.7486],[690,0.7378],[691,0.7373],[695,0.7122],[696,0.7115],[697,0.705],[698,0.6915],[699,0.6854],[700,0.6859],[703,0.6706],[704,0.667],[705,0.6696],[706,0.6662],[709,0.6613],[710,0.6617],[714,0.6587],[715,0.6597],[716,0.6582],[725,0.6552],[739,0.6486],[740,0.6491],[741,0.6463],[742,0.6457],[743,0.642],[745,0.6411],[746,0.6422],[750,0.6397],[751,0.6411],[752,0.6368],[754,0.6381],[755,0.6359],[759,0.6382],[770,0.816],[772,0.8147],[773,0.8112],[774,0.813],[776,0.805],[777,0.8063],[778,0.8053],[779,0.8007],[780,0.7989],[781,0.7993],[785,0.7874],[787,0.7789],[788,0.7758],[790,0.7667],[791,0.767],[792,0.7582],[793,0.7523],[794,0.7483],[795,0.7423],[796,0.7423],[797,0.7336],[798,0.7281],[800,0.7191],[801,0.72],[802,0.7122],[803,0.708],[837,0.8043],[887,0.7423],[888,0.7423],[889,0.7336],[890,0.7281],[892,0.7191],[893,0.72],[894,0.7122],[895,0.708],[929,0.8043]]
       }
       head_ik_ctrl.Location.Z {
        type=MovieSceneScriptingFloatChannel
        default=0
        thinned_from=751
        keys [[-2,-0.1656],[-1,-0.1584],[0,-0.1537],[1,-0.1504],[3,-0.1419],[4,-0.1392],[5,-0.1341],[6,-0.1308],[7,-0.1279],[9,-0.1176],[10,-0.1107],[11,-0.1027],[12,-0.0922],[13,-0.0824],[14,-0.0717],[15,-0.0606],[16,-0.0493],[17,-0.0384],[18,-0.0292],[19,-0.0189],[20,-0.0076],[21,0.001],[22,0.0065],[23,0.0126],[24,0.0175],[25,0.021],[26,0.024],[27,0.025],[28,0.0233],[29,0.0254],[30,0.0266],[31,0.0272],[32,0.0261],[33,0.026],[34,0.0244],[35,0.0243],[36,0.0223],[38,0.0192],[39,0.0184],[40,0.016],[41,0.0152],[42,0.0128],[43,0.0113],[44,0.0105],[45,0.008],[46,0.0071],[47,0.0048],[49,0.0036],[50,0.0023],[51,0.0018],[54,-0.0013],[57,-0.0052],[58,-0.0067],[59,-0.0051],[61,-0.008],[62,-0.0098],[63,-0.0102],[64,-0.0117],[65,-0.0118],[66,-0.0133],[68,-0.0144],[70,-0.0172],[71,-0.0172],[72,-0.0185],[73,-0.018],[74,-0.0192],[75,-0.0193],[76,-0.0208],[77,-0.0198],[79,-0.0206],[81,-0.0202],[82,-0.0202],[83,-0.0197],[84,-0.0188],[85,-0.0182],[87,-0.0153],[88,-0.014],[89,-0.012],[90,-0.0098],[91,-0.0068],[92,-0.0034],[93,0.0005],[94,0.0042],[96,0.0101],[97,0.0103],[98,0.0116],[99,0.0099],[100,0.0095],[101,0.0064],[102,0.0029],[103,0.001],[105,-0.0085],[106,-0.0151],[107,-0.023],[108,-0.0287],[109,-0.036],[110,-0.0444],[112,-0.0601],[113,-0.0663],[114,-0.0745],[115,-0.0805],[116,-0.0881],[117,-0.0963],[118,-0.1018],[119,-0.1087],[120,-0.112],[122,-0.12],[123,-0.123],[124,-0.1252],[125,-0.1285],[126,-0.1294],[127,-0.1313],[128,-0.1325],[130,-0.1336],[132,-0.1342],[133,-0.1349],[134,-0.1341],[135,-0.1309],[136,-0.1284],[138,-0.1253],[139,-0.1232],[140,-0.1214],[141,-0.1191],[142,-0.1171],[143,-0.1146],[145,-0.1083],[146,-0.1058],[147,-0.103],[150,-0.0941],[152,-0.0893],[153,-0.0866],[154,-0.0849],[155,-0.0825],[156,-0.081],[157,-0.08],[158,-0.0792],[159,-0.0776],[161,-0.0769],[162,-0.0768],[164,-0.0769],[165,-0.0827],[166,-0.0813],[167,-0.0819],[168,-0.0797],[169,-0.084],[170,-0.0832],[171,-0.0804],[172,-0.0849],[173,-0.0843],[174,-0.0818],[175,-0.0806],[176,-0.0783],[177,-0.0765],[178,-0.074],[179,-0.0826],[180,-0.0818],[181,-0.0812],[182,-0.0815],[183,-0.0825],[184,-0.084],[185,-0.0863],[187,-0.0892],[191,-0.0941],[192,-0.0951],[193,-0.0963],[196,-0.0985],[197,-0.0998],[198,-0.1006],[199,-0.1097],[202,-0.1108],[203,-0.1117],[206,-0.1123],[207,-0.1132],[209,-0.113],[210,-0.112],[211,-0.1113],[212,-0.1116],[213,-0.1104],[214,-0.0983],[215,-0.0964],[216,-0.0978],[217,-0.0984],[218,-0.086],[219,-0.0861],[220,-0.0856],[221,-0.0736],[222,-0.0748],[223,-0.0729],[224,-0.0719],[225,-0.073],[226,-0.075],[228,-0.0861],[229,-0.095],[230,-0.1044],[231,-0.1159],[232,-0.1264],[233,-0.1391],[234,-0.1498],[235,-0.1613],[236,-0.1695],[237,-0.1753],[238,-0.1815],[239,-0.1853],[240,-0.1873],[241,-0.1997],[243,-0.2038],[244,-0.2048],[245,-0.2159],[246,-0.2174],[248,-0.2188],[249,-0.2182],[250,-0.2283],[251,-0.2281],[252,-0.2272],[253,-0.2254],[254,-0.2247],[255,-0.2221],[257,-0.2187],[258,-0.2159],[259,-0.2112],[260,-0.2095],[261,-0.2042],[262,-0.2026],[263,-0.1981],[264,-0.1975],[265,-0.1941],[266,-0.1938],[267,-0.1929],[268,-0.1936],[269,-0.193],[270,-0.1943],[271,-0.194],[273,-0.1916],[274,-0.1965],[276,-0.1935],[277,-0.1924],[278,-0.197],[280,-0.1949],[281,-0.1947],[282,-0.1927],[283,-0.1918],[284,-0.1964],[285,-0.1959],[286,-0.1941],[287,-0.193],[288,-0.1923],[289,-0.1934],[290,-0.1927],[291,-0.1905],[292,-0.1914],[293,-0.1901],[294,-0.1878],[295,-0.1874],[296,-0.1876],[297,-0.1853],[298,-0.1842],[299,-0.1817],[300,-0.1797],[301,-0.1785],[302,-0.1758],[303,-0.174],[304,-0.1728],[305,-0.1681],[306,-0.1682],[307,-0.1659],[308,-0.1648],[309,-0.1596],[323,-0.1656],[324,-0.1584],[325,-0.1537],[326,-0.1504],[328,-0.1419],[329,-0.1392],[330,-0.1341],[331,-0.1308],[332,-0.1279],[334,-0.1176],[335,-0.1107],[336,-0.1027],[337,-0.0922],[338,-0.0824],[339,-0.0717],[340,-0.0606],[341,-0.0493],[342,-0.0384],[343,-0.0292],[344,-0.0189],[345,-0.0076],[346,0.001],[347,0.0065],[348,0.0126],[349,0.0175],[350,0.021],[351,0.024],[352,0.025],[353,0.0233],[354,0.0254],[355,0.0266],[356,0.0272],[357,0.0261],[358,0.026],[359,0.0244],[360,0.0243],[361,0.0223],[363,0.0192],[364,0.0184],[365,0.016],[366,0.0152],[367,0.0128],[368,0.0113],[369,0.0105],[370,0.008],[371,0.0071],[372,0.0048],[374,0.0036],[375,0.0023],[376,0.0018],[379,-0.0013],[382,-0.0052],[383,-0.0067],[384,-0.0051],[386,-0.008],[387,-0.0098],[388,-0.0102],[389,-0.0117],[390,-0.0118],[391,-0.0133],[393,-0.0144],[395,-0.0172],[396,-0.0172],[397,-0.0185],[398,-0.018],[399,-0.0192],[400,-0.0193],[401,-0.0208],[402,-0.0198],[404,-0.0206],[406,-0.0202],[407,-0.0202],[408,-0.0197],[409,-0.0188],[410,-0.0182],[412,-0.0153],[413,-0.014],[414,-0.012],[415,-0.0098],[416,-0.0068],[417,-0.0034],[418,0.0005],[419,0.0042],[421,0.0101],[422,0.0103],[423,0.0116],[424,0.0099],[425,0.0095],[426,0.0064],[427,0.0029],[428,0.001],[430,-0.0085],[431,-0.0151],[432,-0.023],[433,-0.0287],[434,-0.036],[435,-0.0444],[437,-0.0601],[438,-0.0663],[439,-0.0745],[440,-0.0805],[441,-0.0881],[442,-0.0963],[443,-0.1018],[444,-0.1087],[445,-0.112],[447,-0.12],[448,-0.123],[449,-0.1252],[450,-0.1285],[451,-0.1294],[452,-0.1313],[453,-0.1325],[455,-0.1336],[457,-0.1342],[458,-0.1349],[459,-0.1341],[460,-0.1309],[461,-0.1284],[463,-0.1253],[464,-0.1232],[465,-0.1214],[466,-0.1191],[467,-0.1171],[468,-0.1146],[470,-0.1083],[471,-0.1058],[472,-0.103],[475,-0.0941],[477,-0.0893],[478,-0.0866],[479,-0.0849],[480,-0.0825],[481,-0.081],[482,-0.08],[483,-0.0792],[484,-0.0776],[486,-0.0769],[487,-0.0768],[489,-0.0769],[490,-0.0827],[491,-0.0813],[492,-0.0819],[493,-0.0797],[494,-0.084],[495,-0.0832],[496,-0.0804],[497,-0.0849],[498,-0.0843],[499,-0.0818],[500,-0.0806],[501,-0.0783],[502,-0.0765],[503,-0.074],[504,-0.0826],[505,-0.0818],[506,-0.0812],[507,-0.0815],[508,-0.0825],[509,-0.084],[510,-0.0863],[512,-0.0892],[516,-0.0941],[517,-0.0951],[518,-0.0963],[521,-0.0985],[522,-0.0998],[523,-0.1006],[524,-0.1097],[527,-0.1108],[528,-0.1117],[531,-0.1123],[532,-0.1132],[534,-0.113],[535,-0.112],[536,-0.1113],[537,-0.1116],[538,-0.1104],[539,-0.0983],[540,-0.0964],[541,-0.0978],[542,-0.0984],[543,-0.086],[544,-0.0861],[545,-0.0856],[546,-0.0736],[547,-0.0748],[548,-0.0729],[549,-0.0719],[550,-0.073],[551,-0.075],[553,-0.0861],[554,-0.095],[555,-0.1044],[556,-0.1159],[557,-0.1264],[558,-0.1391],[559,-0.1498],[560,-0.1613],[561,-0.1695],[562,-0.1753],[563,-0.1815],[564,-0.1853],[565,-0.1873],[566,-0.1997],[568,-0.2038],[569,-0.2048],[570,-0.2159],[571,-0.2174],[573,-0.2188],[574,-0.2182],[575,-0.2283],[576,-0.2281],[577,-0.2272],[578,-0.2254],[579,-0.2247],[580,-0.2221],[582,-0.2187],[583,-0.2159],[584,-0.2112],[585,-0.2095],[586,-0.2042],[587,-0.2026],[588,-0.1981],[589,-0.1975],[590,-0.1941],[591,-0.1938],[592,-0.1929],[593,-0.1936],[594,-0.193],[595,-0.1943],[596,-0.194],[598,-0.1916],[599,-0.1965],[601,-0.1935],[602,-0.1924],[603,-0.197],[605,-0.1949],[606,-0.1947],[607,-0.1927],[608,-0.1918],[609,-0.1964],[610,-0.1959],[611,-0.1941],[612,-0.193],[613,-0.1923],[614,-0.1934],[615,-0.1927],[616,-0.1905],[617,-0.1914],[618,-0.1901],[619,-0.1878],[620,-0.1874],[621,-0.1876],[622,-0.1853],[623,-0.1842],[624,-0.1817],[625,-0.1797],[626,-0.1785],[627,-0.1758],[628,-0.174],[629,-0.1728],[630,-0.1681],[631,-0.1682],[632,-0.1659],[633,-0.1648],[634,-0.1596],[676,-0.1656],[677,-0.1584],[678,-0.1537],[679,-0.1504],[681,-0.1419],[682,-0.1392],[683,-0.1341],[684,-0.1308],[685,-0.1279],[687,-0.1176],[688,-0.1107],[689,-0.1027],[690,-0.0922],[691,-0.0824],[692,-0.0717],[693,-0.0606],[694,-0.0493],[695,-0.0384],[696,-0.0292],[697,-0.0189],[698,-0.0076],[699,0.001],[700,0.0065],[701,0.0126],[702,0.0175],[703,0.021],[704,0.024],[705,0.025],[706,0.0233],[707,0.0254],[708,0.0266],[709,0.0272],[710,0.0261],[711,0.026],[712,0.0244],[713,0.0243],[714,0.0223],[716,0.0192],[717,0.0184],[718,0.016],[719,0.0152],[720,0.0128],[721,0.0113],[722,0.0105],[723,0.008],[724,0.0071],[725,0.0048],[727,0.0036],[728,0.0023],[729,0.0018],[732,-0.0013],[735,-0.0052],[736,-0.0067],[737,-0.0051],[739,-0.008],[740,-0.0098],[741,-0.0102],[742,-0.0117],[743,-0.0118],[744,-0.0133],[746,-0.0144],[748,-0.0172],[749,-0.0172],[750,-0.0185],[751,-0.018],[752,-0.0192],[753,-0.0193],[754,-0.0208],[755,-0.0198],[757,-0.0206],[759,-0.0202],[770,-0.1959],[771,-0.1941],[772,-0.193],[773,-0.1923],[774,-0.1934],[775,-0.1927],[776,-0.1905],[777,-0.1914],[778,-0.1901],[779,-0.1878],[780,-0.1874],[781,-0.1876],[782,-0.1853],[783,-0.1842],[784,-0.1817],[785,-0.1797],[786,-0.1785],[787,-0.1758],[788,-0.174],[789,-0.1728],[790,-0.1681],[791,-0.1682],[792,-0.1659],[793,-0.1648],[794,-0.1596],[795,-0.158],[796,-0.1578],[797,-0.1549],[798,-0.1537],[799,-0.1482],[800,-0.1464],[801,-0.1458],[802,-0.1432],[803,-0.1415],[837,-0.1656],[888,-0.1578],[889,-0.1549],[890,-0.1537],[891,-0.1482],[892,-0.1464],[893,-0.1458],[894,-0.1432],[895,-0.1415],[929,-0.1656]]
       }
       head_ik_ctrl.Rotation.X {
        type=MovieSceneScriptingFloatChannel
        default=0
        thinned_from=751
        keys [[-2,-5.7506],[-1,-5.3693],[0,-5.2561],[1,-5.176],[2,-5.1261],[3,-5.1044],[4,-5.1087],[5,-5.1392],[6,-5.1867],[7,-5.2626],[8,-5.3614],[9,-5.4735],[10,-5.6053],[12,-5.9037],[16,-6.5295],[18,-6.796],[19,-6.9138],[21,-7.0911],[23,-7.2042],[25,-7.2423],[27,-7.2095],[28,-7.1671],[30,-7.0371],[31,-6.9529],[33,-6.7421],[35,-6.5027],[43,-5.4865],[45,-5.2714],[48,-5],[50,-4.8467],[51,-4.7907],[52,-4.7468],[53,-4.6932],[57,-4.5697],[61,-4.4867],[64,-4.4509],[66,-4.445],[68,-4.4731],[69,-4.5143],[70,-4.5662],[71,-4.6379],[72,-4.7273],[73,-4.8445],[74,-4.9927],[75,-5.1722],[76,-5.4146],[77,-5.734],[78,-6.1024],[79,-6.5142],[80,-6.9631],[84,-8.8483],[85,-9.2909],[86,-9.6959],[87,-10.0457],[89,-10.6073],[91,-11.1363],[92,-11.3839],[93,-11.6016],[94,-11.7905],[95,-11.9385],[96,-12.0133],[97,-12.0024],[98,-11.9288],[99,-11.795],[100,-11.6122],[101,-11.3861],[104,-10.5575],[106,-9.9856],[109,-9.2179],[111,-8.7376],[113,-8.3226],[114,-8.1505],[115,-8.0299],[116,-7.9572],[118,-7.9414],[119,-7.9886],[121,-8.1901],[123,-8.4954],[124,-8.6691],[127,-9.2717],[130,-9.8981],[132,-10.2781],[133,-10.4496],[134,-10.595],[135,-10.7074],[136,-10.7914],[138,-10.8528],[142,-10.8555],[155,-10.5182],[160,-10.2885],[161,-10.2673],[162,-10.2178],[163,-10.1048],[164,-9.9324],[166,-9.4927],[168,-8.9763],[169,-8.7342],[170,-8.5223],[171,-8.3549],[172,-8.2474],[173,-8.2024],[175,-8.2583],[178,-8.4593],[179,-8.5421],[182,-8.73],[185,-8.8668],[186,-8.8953],[187,-8.8513],[189,-8.8697],[190,-8.8629],[192,-8.8012],[194,-8.6885],[198,-8.3724],[202,-8.0075],[204,-7.851],[206,-7.7439],[207,-7.7102],[208,-7.6924],[209,-7.7072],[210,-7.7866],[211,-7.8931],[213,-8.1604],[214,-8.3321],[215,-8.5437],[216,-8.7959],[218,-9.3612],[220,-9.9675],[221,-10.2557],[222,-10.5224],[223,-10.7637],[224,-10.9745],[225,-11.1566],[226,-11.3075],[227,-11.4346],[228,-11.5334],[229,-11.6017],[230,-11.6351],[231,-11.6355],[232,-11.6012],[233,-11.5107],[234,-11.3848],[235,-11.215],[236,-11.0103],[237,-10.7527],[238,-10.4345],[239,-10.0553],[240,-9.6252],[241,-9.1451],[246,-6.4636],[247,-5.9437],[248,-5.4479],[249,-4.9754],[250,-4.5306],[251,-4.1182],[252,-3.7351],[253,-3.3834],[254,-3.0497],[255,-2.7342],[256,-2.4255],[258,-1.7893],[259,-1.4648],[261,-0.832],[262,-0.5206],[263,-0.2159],[264,0.0844],[265,0.3797],[266,0.6694],[267,0.9634],[269,1.5318],[271,2.0735],[272,2.3302],[273,2.5755],[274,2.809],[278,3.6968],[280,4.1183],[282,4.5017],[284,4.8405],[286,5.1551],[288,5.4508],[290,5.7196],[291,5.8251],[292,5.9098],[293,5.9668],[294,5.9944],[295,6.0051],[296,5.9897],[297,5.9418],[298,5.8699],[299,5.7766],[300,5.6661],[301,5.5381],[302,5.3923],[303,5.2354],[304,5.0673],[309,4.1445],[323,-5.7506],[324,-5.3693],[325,-5.2561],[326,-5.176],[327,-5.1261],[328,-5.1044],[329,-5.1087],[330,-5.1392],[331,-5.1867],[332,-5.2626],[333,-5.3614],[334,-5.4735],[335,-5.6053],[337,-5.9037],[341,-6.5295],[343,-6.796],[344,-6.9138],[346,-7.0911],[348,-7.2042],[350,-7.2423],[352,-7.2095],[353,-7.1671],[355,-7.0371],[356,-6.9529],[358,-6.7421],[360,-6.5027],[368,-5.4865],[370,-5.2714],[373,-5],[375,-4.8467],[376,-4.7907],[377,-4.7468],[378,-4.6932],[382,-4.5697],[386,-4.4867],[389,-4.4509],[391,-4.445],[393,-4.4731],[394,-4.5143],[395,-4.5662],[396,-4.6379],[397,-4.7273],[398,-4.8445],[399,-4.9927],[400,-5.1722],[401,-5.4146],[402,-5.734],[403,-6.1024],[404,-6.5142],[405,-6.9631],[409,-8.8483],[410,-9.2909],[411,-9.6959],[412,-10.0457],[414,-10.6073],[416,-11.1363],[417,-11.3839],[418,-11.6016],[419,-11.7905],[420,-11.9385],[421,-12.0133],[422,-12.0024],[423,-11.9288],[424,-11.795],[425,-11.6122],[426,-11.3861],[429,-10.5575],[431,-9.9856],[434,-9.2179],[436,-8.7376],[438,-8.3226],[439,-8.1505],[440,-8.0299],[441,-7.9572],[443,-7.9414],[444,-7.9886],[446,-8.1901],[448,-8.4954],[449,-8.6691],[452,-9.2717],[455,-9.8981],[457,-10.2781],[458,-10.4496],[459,-10.595],[460,-10.7074],[461,-10.7914],[463,-10.8528],[467,-10.8555],[480,-10.5182],[485,-10.2885],[486,-10.2673],[487,-10.2178],[488,-10.1048],[489,-9.9324],[491,-9.4927],[493,-8.9763],[494,-8.7342],[495,-8.5223],[496,-8.3549],[497,-8.2474],[498,-8.2024],[500,-8.2583],[503,-8.4593],[504,-8.5421],[507,-8.73],[510,-8.8668],[511,-8.8953],[512,-8.8513],[514,-8.8697],[515,-8.8629],[517,-8.8012],[519,-8.6885],[523,-8.3724],[527,-8.0075],[529,-7.851],[531,-7.7439],[532,-7.7102],[533,-7.6924],[534,-7.7072],[535,-7.7866],[536,-7.8931],[538,-8.1604],[539,-8.3321],[540,-8.5437],[541,-8.7959],[543,-9.3612],[545,-9.9675],[546,-10.2557],[547,-10.5224],[548,-10.7637],[549,-10.9745],[550,-11.1566],[551,-11.3075],[552,-11.4346],[553,-11.5334],[554,-11.6017],[555,-11.6351],[556,-11.6355],[557,-11.6012],[558,-11.5107],[559,-11.3848],[560,-11.215],[561,-11.0103],[562,-10.7527],[563,-10.4345],[564,-10.0553],[565,-9.6252],[566,-9.1451],[571,-6.4636],[572,-5.9437],[573,-5.4479],[574,-4.9754],[575,-4.5306],[576,-4.1182],[577,-3.7351],[578,-3.3834],[579,-3.0497],[580,-2.7342],[581,-2.4255],[583,-1.7893],[584,-1.4648],[586,-0.832],[587,-0.5206],[588,-0.2159],[589,0.0844],[590,0.3797],[591,0.6694],[592,0.9634],[594,1.5318],[596,2.0735],[597,2.3302],[598,2.5755],[599,2.809],[603,3.6968],[605,4.1183],[607,4.5017],[609,4.8405],[611,5.1551],[613,5.4508],[615,5.7196],[616,5.8251],[617,5.9098],[618,5.9668],[619,5.9944],[620,6.0051],[621,5.9897],[622,5.9418],[623,5.8699],[624,5.7766],[625,5.6661],[626,5.5381],[627,5.3923],[628,5.2354],[629,5.0673],[634,4.1445],[676,-5.7506],[677,-5.3693],[678,-5.2561],[679,-5.176],[680,-5.1261],[681,-5.1044],[682,-5.1087],[683,-5.1392],[684,-5.1867],[685,-5.2626],[686,-5.3614],[687,-5.4735],[688,-5.6053],[690,-5.9037],[694,-6.5295],[696,-6.796],[697,-6.9138],[699,-7.0911],[701,-7.2042],[703,-7.2423],[705,-7.2095],[706,-7.1671],[708,-7.0371],[709,-6.9529],[711,-6.7421],[713,-6.5027],[721,-5.4865],[723,-5.2714],[726,-5],[728,-4.8467],[729,-4.7907],[730,-4.7468],[731,-4.6932],[735,-4.5697],[739,-4.4867],[742,-4.4509],[744,-4.445],[746,-4.4731],[747,-4.5143],[748,-4.5662],[749,-4.6379],[750,-4.7273],[751,-4.8445],[752,-4.9927],[753,-5.1722],[754,-5.4146],[755,-5.734],[757,-6.5142],[759,-7.4331],[770,5.0004],[773,5.4508],[775,5.7196],[776,5.8251],[777,5.9098],[778,5.9668],[779,5.9944],[780,6.0051],[781,5.9897],[782,5.9418],[783,5.8699],[784,5.7766],[785,5.6661],[786,5.5381],[787,5.3923],[788,5.2354],[789,5.0673],[794,4.1445],[796,3.798],[797,3.6371],[798,3.4907],[799,3.3567],[800,3.2451],[801,3.1527],[802,3.0773],[803,3.0159],[837,-5.7506],[887,3.9675],[888,3.798],[889,3.6371],[890,3.4907],[891,3.3567],[892,3.2451],[893,3.1527],[894,3.0773],[895,3.0159],[929,-5.7506]]
       }
       head_ik_ctrl.Rotation.Y {
        type=MovieSceneScriptingFloatChannel
        default=0
        thinned_from=751
        keys [[-2,-4.3117],[-1,-3.4371],[0,-3.0805],[1,-2.7387],[2,-2.4155],[3,-2.1057],[4,-1.8132],[5,-1.5327],[6,-1.2654],[7,-1.0114],[8,-0.7351],[9,-0.4007],[10,0.0085],[11,0.4682],[12,0.9722],[13,1.5186],[14,2.0984],[15,2.7007],[18,4.5803],[19,5.1965],[20,5.7862],[21,6.3423],[22,6.8525],[23,7.2913],[24,7.6628],[25,7.9697],[26,8.2364],[27,8.4664],[28,8.6489],[29,8.7942],[30,8.9023],[31,8.9823],[34,9.1164],[38,9.196],[45,9.1587],[49,9.0867],[50,9.0435],[51,9.0211],[54,8.8623],[57,8.6379],[66,7.7767],[71,7.2183],[73,6.9602],[75,6.665],[76,6.484],[77,6.265],[78,6.0198],[80,5.4756],[81,5.1901],[83,4.6003],[84,4.3351],[85,4.1482],[86,4.0503],[87,4.0319],[88,4.1458],[89,4.3656],[90,4.6501],[91,4.9799],[93,5.6639],[94,5.9786],[95,6.2358],[96,6.4087],[97,6.4781],[98,6.4535],[99,6.3643],[100,6.1971],[101,5.9501],[102,5.6313],[103,5.2144],[104,4.6984],[105,4.1038],[106,3.4618],[107,2.7769],[108,2.0674],[110,0.6307],[111,-0.0813],[112,-0.7708],[113,-1.4526],[114,-2.12],[115,-2.7617],[116,-3.3663],[117,-3.9127],[118,-4.402],[119,-4.817],[120,-5.1478],[121,-5.4115],[122,-5.6102],[123,-5.7475],[124,-5.8119],[125,-5.8239],[126,-5.7991],[127,-5.7535],[131,-5.4545],[132,-5.3875],[134,-5.2828],[137,-5.1634],[142,-4.995],[150,-4.715],[153,-4.6456],[155,-4.6271],[157,-4.6353],[160,-4.6954],[161,-4.7229],[162,-4.7848],[163,-4.9003],[164,-5.049],[165,-5.2282],[169,-6.0489],[170,-6.2447],[171,-6.4136],[172,-6.5397],[173,-6.6229],[174,-6.6775],[175,-6.7076],[176,-6.686],[177,-6.6401],[179,-6.5],[183,-6.1695],[184,-6.1048],[189,-5.6763],[191,-5.5239],[193,-5.392],[196,-5.2392],[198,-5.1798],[200,-5.1564],[206,-5.1896],[207,-5.1836],[208,-5.1536],[209,-5.0176],[210,-4.653],[211,-4.2348],[212,-3.7891],[213,-3.3346],[214,-2.8723],[215,-2.3949],[216,-1.9006],[217,-1.4337],[218,-1.0046],[219,-0.6509],[220,-0.3697],[221,-0.1862],[222,-0.106],[223,-0.1058],[224,-0.1722],[225,-0.3065],[226,-0.5024],[227,-0.7501],[228,-1.0432],[229,-1.3796],[230,-1.7414],[231,-2.1239],[232,-2.5003],[233,-2.8859],[234,-3.289],[235,-3.7196],[236,-4.1787],[237,-4.6524],[238,-5.1406],[240,-6.1383],[242,-7.1103],[243,-7.5836],[244,-8.0403],[245,-8.473],[246,-8.8732],[247,-9.2363],[248,-9.5343],[249,-9.7693],[250,-9.9176],[251,-9.9806],[252,-9.9792],[253,-9.9214],[254,-9.8208],[255,-9.6805],[256,-9.5156],[257,-9.3257],[261,-8.4573],[262,-8.279],[263,-8.1201],[264,-7.9959],[265,-7.9051],[266,-7.8443],[267,-7.8206],[271,-7.7895],[283,-7.4721],[284,-7.4993],[285,-7.613],[286,-7.8037],[287,-8.0413],[288,-8.3258],[290,-8.9448],[291,-9.2363],[292,-9.493],[293,-9.6999],[294,-9.8474],[295,-9.9742],[296,-10.0536],[297,-10.0907],[298,-10.0923],[299,-10.0696],[300,-10.0205],[301,-9.95],[303,-9.7443],[305,-9.4604],[309,-8.7649],[323,-4.3117],[324,-3.4371],[325,-3.0805],[326,-2.7387],[327,-2.4155],[328,-2.1057],[329,-1.8132],[330,-1.5327],[331,-1.2654],[332,-1.0114],[333,-0.7351],[334,-0.4007],[335,0.0085],[336,0.4682],[337,0.9722],[338,1.5186],[339,2.0984],[340,2.7007],[343,4.5803],[344,5.1965],[345,5.7862],[346,6.3423],[347,6.8525],[348,7.2913],[349,7.6628],[350,7.9697],[351,8.2364],[352,8.4664],[353,8.6489],[354,8.7942],[355,8.9023],[356,8.9823],[359,9.1164],[363,9.196],[370,9.1587],[374,9.0867],[375,9.0435],[376,9.0211],[379,8.8623],[382,8.6379],[391,7.7767],[396,7.2183],[398,6.9602],[400,6.665],[401,6.484],[402,6.265],[403,6.0198],[405,5.4756],[406,5.1901],[408,4.6003],[409,4.3351],[410,4.1482],[411,4.0503],[412,4.0319],[413,4.1458],[414,4.3656],[415,4.6501],[416,4.9799],[418,5.6639],[419,5.9786],[420,6.2358],[421,6.4087],[422,6.4781],[423,6.4535],[424,6.3643],[425,6.1971],[426,5.9501],[427,5.6313],[428,5.2144],[429,4.6984],[430,4.1038],[431,3.4618],[432,2.7769],[433,2.0674],[435,0.6307],[436,-0.0813],[437,-0.7708],[438,-1.4526],[439,-2.12],[440,-2.7617],[441,-3.3663],[442,-3.9127],[443,-4.402],[444,-4.817],[445,-5.1478],[446,-5.4115],[447,-5.6102],[448,-5.7475],[449,-5.8119],[450,-5.8239],[451,-5.7991],[452,-5.7535],[456,-5.4545],[457,-5.3875],[459,-5.2828],[462,-5.1634],[467,-4.995],[475,-4.715],[478,-4.6456],[480,-4.6271],[482,-4.6353],[485,-4.6954],[486,-4.7229],[487,-4.7848],[488,-4.9003],[489,-5.049],[490,-5.2282],[494,-6.0489],[495,-6.2447],[496,-6.4136],[497,-6.5397],[498,-6.6229],[499,-6.6775],[500,-6.7076],[501,-6.686],[502,-6.6401],[504,-6.5],[508,-6.1695],[509,-6.1048],[514,-5.6763],[516,-5.5239],[518,-5.392],[521,-5.2392],[523,-5.1798],[525,-5.1564],[531,-5.1896],[532,-5.1836],[533,-5.1536],[534,-5.0176],[535,-4.653],[536,-4.2348],[537,-3.7891],[538,-3.3346],[539,-2.8723],[540,-2.3949],[541,-1.9006],[542,-1.4337],[543,-1.0046],[544,-0.6509],[545,-0.3697],[546,-0.1862],[547,-0.106],[548,-0.1058],[549,-0.1722],[550,-0.3065],[551,-0.5024],[552,-0.7501],[553,-1.0432],[554,-1.3796],[555,-1.7414],[556,-2.1239],[557,-2.5003],[558,-2.8859],[559,-3.289],[560,-3.7196],[561,-4.1787],[562,-4.6524],[563,-5.1406],[565,-6.1383],[567,-7.1103],[568,-7.5836],[569,-8.0403],[570,-8.473],[571,-8.8732],[572,-9.2363],[573,-9.5343],[574,-9.7693],[575,-9.9176],[576,-9.9806],[577,-9.9792],[578,-9.9214],[579,-9.8208],[580,-9.6805],[581,-9.5156],[582,-9.3257],[586,-8.4573],[587,-8.279],[588,-8.1201],[589,-7.9959],[590,-7.9051],[591,-7.8443],[592,-7.8206],[596,-7.7895],[608,-7.4721],[609,-7.4993],[610,-7.613],[611,-7.8037],[612,-8.0413],[613,-8.3258],[615,-8.9448],[616,-9.2363],[617,-9.493],[618,-9.6999],[619,-9.8474],[620,-9.9742],[621,-10.0536],[622,-10.0907],[623,-10.0923],[624,-10.0696],[625,-10.0205],[626,-9.95],[628,-9.7443],[630,-9.4604],[634,-8.7649],[676,-4.3117],[677,-3.4371],[678,-3.0805],[679,-2.7387],[680,-2.4155],[681,-2.1057],[682,-1.8132],[683,-1.5327],[684,-1.2654],[685,-1.0114],[686,-0.7351],[687,-0.4007],[688,0.0085],[689,0.4682],[690,0.9722],[691,1.5186],[692,2.0984],[693,2.7007],[696,4.5803],[697,5.1965],[698,5.7862],[699,6.3423],[700,6.8525],[701,7.2913],[702,7.6628],[703,7.9697],[704,8.2364],[705,8.4664],[706,8.6489],[707,8.7942],[708,8.9023],[709,8.9823],[712,9.1164],[716,9.196],[723,9.1587],[727,9.0867],[728,9.0435],[729,9.0211],[732,8.8623],[735,8.6379],[744,7.7767],[749,7.2183],[751,6.9602],[753,6.665],[754,6.484],[755,6.265],[757,5.7531],[759,5.1901],[770,-7.613],[771,-7.8037],[772,-8.0413],[773,-8.3258],[775,-8.9448],[776,-9.2363],[777,-9.493],[778,-9.6999],[779,-9.8474],[780,-9.9742],[781,-10.0536],[782,-10.0907],[783,-10.0923],[784,-10.0696],[785,-10.0205],[786,-9.95],[788,-9.7443],[790,-9.4604],[798,-8.0636],[800,-7.7477],[803,-7.3434],[837,-4.3117],[887,-8.5858],[890,-8.0636],[892,-7.7477],[895,-7.3434],[929,-4.3117]]
       }
       head_ik_ctrl.Rotation.Z {
        type=MovieSceneScriptingFloatChannel
        default=0
        thinned_from=751
        keys [[-2,-1.2578],[-1,-0.6105],[0,-0.3839],[1,-0.1608],[2,0.0568],[3,0.2686],[4,0.471],[5,0.6648],[6,0.851],[8,1.1988],[10,1.5297],[11,1.6877],[14,2.1206],[15,2.2559],[16,2.3826],[17,2.4969],[18,2.5887],[19,2.6684],[20,2.7273],[21,2.7742],[22,2.7933],[23,2.7916],[24,2.7742],[25,2.7435],[26,2.7033],[27,2.6509],[29,2.5284],[32,2.3183],[35,2.1252],[38,1.9451],[40,1.8342],[42,1.7384],[44,1.654],[45,1.6183],[46,1.5869],[47,1.56],[49,1.5209],[51,1.5031],[53,1.5073],[54,1.5165],[55,1.5309],[57,1.5706],[59,1.6218],[63,1.7493],[66,1.86],[67,1.8707],[68,1.8611],[69,1.8227],[70,1.7484],[71,1.6283],[72,1.4608],[73,1.2294],[74,0.9184],[75,0.4944],[76,-0.1238],[77,-0.9889],[78,-2.0326],[79,-3.2403],[80,-4.5462],[81,-5.9166],[82,-7.3344],[83,-8.7993],[84,-10.2425],[85,-11.6077],[86,-12.8335],[87,-13.874],[88,-14.6115],[89,-15.1795],[90,-15.6492],[91,-16.023],[92,-16.3106],[93,-16.5001],[94,-16.596],[95,-16.5822],[96,-16.4447],[97,-16.1859],[98,-15.8311],[99,-15.3622],[100,-14.8019],[101,-14.1524],[102,-13.4096],[104,-11.8296],[106,-10.3101],[107,-9.5825],[108,-8.8932],[109,-8.2422],[110,-7.6351],[111,-7.0779],[112,-6.5848],[113,-6.1333],[114,-5.7094],[115,-5.3274],[116,-4.9939],[117,-4.7095],[118,-4.4838],[119,-4.3182],[120,-4.2231],[121,-4.1824],[122,-4.1759],[123,-4.2076],[124,-4.289],[125,-4.4169],[126,-4.5994],[127,-4.8173],[129,-5.2975],[130,-5.523],[131,-5.7198],[132,-5.8937],[133,-6.0403],[134,-6.1566],[135,-6.2442],[136,-6.3039],[137,-6.3479],[139,-6.3673],[140,-6.3533],[142,-6.2862],[144,-6.17],[150,-5.7242],[152,-5.6026],[153,-5.5597],[154,-5.5304],[155,-5.5176],[156,-5.523],[157,-5.5482],[158,-5.5933],[159,-5.6646],[160,-5.7626],[161,-5.8912],[162,-6.0813],[163,-6.3647],[164,-6.7271],[165,-7.1446],[166,-7.6054],[169,-9.0778],[170,-9.5394],[171,-9.9542],[172,-10.3073],[173,-10.5927],[174,-10.8257],[176,-11.2262],[178,-11.5207],[180,-11.7265],[181,-11.7999],[182,-11.8465],[184,-11.8568],[190,-11.5147],[202,-11.0205],[203,-10.9302],[204,-10.8169],[205,-10.6671],[206,-10.4861],[207,-10.2556],[208,-9.9384],[209,-9.3679],[210,-8.2434],[211,-6.8972],[212,-5.3531],[213,-3.6502],[214,-1.8416],[215,0.0395],[216,1.9729],[217,3.9248],[218,5.8923],[219,7.8788],[222,13.912],[223,15.8402],[224,17.6879],[225,19.4175],[226,21.0126],[227,22.4473],[228,23.7064],[229,24.7562],[230,25.5815],[231,26.1584],[232,26.49],[233,26.6245],[234,26.5713],[235,26.3494],[236,25.9765],[237,25.4658],[238,24.8454],[239,24.1245],[240,23.3143],[241,22.4357],[243,20.526],[247,16.5258],[248,15.5548],[249,14.6219],[250,13.7426],[251,12.9224],[253,11.4337],[254,10.7128],[255,10.0176],[256,9.3517],[257,8.728],[258,8.1372],[259,7.5868],[260,7.0779],[261,6.6065],[262,6.1684],[263,5.7667],[264,5.409],[265,5.0942],[266,4.8228],[267,4.6119],[269,4.2446],[271,3.9158],[272,3.7668],[273,3.6316],[274,3.5067],[276,3.2866],[277,3.1845],[279,3.0082],[281,2.862],[282,2.7966],[283,2.7216],[284,2.5307],[285,2.0843],[286,1.4565],[287,0.6644],[288,-0.2326],[289,-1.2079],[290,-2.1868],[291,-3.1393],[292,-4.008],[293,-4.7403],[294,-5.2928],[295,-5.8068],[296,-6.1955],[297,-6.4896],[298,-6.7209],[299,-6.8867],[300,-7.0154],[301,-7.0923],[302,-7.1346],[303,-7.1343],[304,-7.0976],[305,-7.0203],[306,-6.9074],[307,-6.7751],[308,-6.6231],[309,-6.4573],[323,-1.2578],[324,-0.6105],[325,-0.3839],[326,-0.1608],[327,0.0568],[328,0.2686],[329,0.471],[330,0.6648],[331,0.851],[333,1.1988],[335,1.5297],[336,1.6877],[339,2.1206],[340,2.2559],[341,2.3826],[342,2.4969],[343,2.5887],[344,2.6684],[345,2.7273],[346,2.7742],[347,2.7933],[348,2.7916],[349,2.7742],[350,2.7435],[351,2.7033],[352,2.6509],[354,2.5284],[357,2.3183],[360,2.1252],[363,1.9451],[365,1.8342],[367,1.7384],[369,1.654],[370,1.6183],[371,1.5869],[372,1.56],[374,1.5209],[376,1.5031],[378,1.5073],[379,1.5165],[380,1.5309],[382,1.5706],[384,1.6218],[388,1.7493],[391,1.86],[392,1.8707],[393,1.8611],[394,1.8227],[395,1.7484],[396,1.6283],[397,1.4608],[398,1.2294],[399,0.9184],[400,0.4944],[401,-0.1238],[402,-0.9889],[403,-2.0326],[404,-3.2403],[405,-4.5462],[406,-5.9166],[407,-7.3344],[408,-8.7993],[409,-10.2425],[410,-11.6077],[411,-12.8335],[412,-13.874],[413,-14.6115],[414,-15.1795],[415,-15.6492],[416,-16.023],[417,-16.3106],[418,-16.5001],[419,-16.596],[420,-16.5822],[421,-16.4447],[422,-16.1859],[423,-15.8311],[424,-15.3622],[425,-14.8019],[426,-14.1524],[427,-13.4096],[429,-11.8296],[431,-10.3101],[432,-9.5825],[433,-8.8932],[434,-8.2422],[435,-7.6351],[436,-7.0779],[437,-6.5848],[438,-6.1333],[439,-5.7094],[440,-5.3274],[441,-4.9939],[442,-4.7095],[443,-4.4838],[444,-4.3182],[445,-4.2231],[446,-4.1824],[447,-4.1759],[448,-4.2076],[449,-4.289],[450,-4.4169],[451,-4.5994],[452,-4.8173],[454,-5.2975],[455,-5.523],[456,-5.7198],[457,-5.8937],[458,-6.0403],[459,-6.1566],[460,-6.2442],[461,-6.3039],[462,-6.3479],[464,-6.3673],[465,-6.3533],[467,-6.2862],[469,-6.17],[475,-5.7242],[477,-5.6026],[478,-5.5597],[479,-5.5304],[480,-5.5176],[481,-5.523],[482,-5.5482],[483,-5.5933],[484,-5.6646],[485,-5.7626],[486,-5.8912],[487,-6.0813],[488,-6.3647],[489,-6.7271],[490,-7.1446],[491,-7.6054],[494,-9.0778],[495,-9.5394],[496,-9.9542],[497,-10.3073],[498,-10.5927],[499,-10.8257],[501,-11.2262],[503,-11.5207],[505,-11.7265],[506,-11.7999],[507,-11.8465],[509,-11.8568],[515,-11.5147],[527,-11.0205],[528,-10.9302],[529,-10.8169],[530,-10.6671],[531,-10.4861],[532,-10.2556],[533,-9.9384],[534,-9.3679],[535,-8.2434],[536,-6.8972],[537,-5.3531],[538,-3.6502],[539,-1.8416],[540,0.0395],[541,1.9729],[542,3.9248],[543,5.8923],[544,7.8788],[547,13.912],[548,15.8402],[549,17.6879],[550,19.4175],[551,21.0126],[552,22.4473],[553,23.7064],[554,24.7562],[555,25.5815],[556,26.1584],[557,26.49],[558,26.6245],[559,26.5713],[560,26.3494],[561,25.9765],[562,25.4658],[563,24.8454],[564,24.1245],[565,23.3143],[566,22.4357],[568,20.526],[572,16.5258],[573,15.5548],[574,14.6219],[575,13.7426],[576,12.9224],[578,11.4337],[579,10.7128],[580,10.0176],[581,9.3517],[582,8.728],[583,8.1372],[584,7.5868],[585,7.0779],[586,6.6065],[587,6.1684],[588,5.7667],[589,5.409],[590,5.0942],[591,4.8228],[592,4.6119],[594,4.2446],[596,3.9158],[597,3.7668],[598,3.6316],[599,3.5067],[601,3.2866],[602,3.1845],[604,3.0082],[606,2.862],[607,2.7966],[608,2.7216],[609,2.5307],[610,2.0843],[611,1.4565],[612,0.6644],[613,-0.2326],[614,-1.2079],[615,-2.1868],[616,-3.1393],[617,-4.008],[618,-4.7403],[619,-5.2928],[620,-5.8068],[621,-6.1955],[622,-6.4896],[623,-6.7209],[624,-6.8867],[625,-7.0154],[626,-7.0923],[627,-7.1346],[628,-7.1343],[629,-7.0976],[630,-7.0203],[631,-6.9074],[632,-6.7751],[633,-6.6231],[634,-6.4573],[676,-1.2578],[677,-0.6105],[678,-0.3839],[679,-0.1608],[680,0.0568],[681,0.2686],[682,0.471],[683,0.6648],[684,0.851],[686,1.1988],[688,1.5297],[689,1.6877],[692,2.1206],[693,2.2559],[694,2.3826],[695,2.4969],[696,2.5887],[697,2.6684],[698,2.7273],[699,2.7742],[700,2.7933],[701,2.7916],[702,2.7742],[703,2.7435],[704,2.7033],[705,2.6509],[707,2.5284],[710,2.3183],[713,2.1252],[716,1.9451],[718,1.8342],[720,1.7384],[722,1.654],[723,1.6183],[724,1.5869],[725,1.56],[727,1.5209],[729,1.5031],[731,1.5073],[732,1.5165],[733,1.5309],[735,1.5706],[737,1.6218],[741,1.7493],[744,1.86],[745,1.8707],[746,1.8611],[747,1.8227],[748,1.7484],[749,1.6283],[750,1.4608],[751,1.2294],[752,0.9184],[753,0.4944],[754,-0.1238],[755,-0.9889],[757,-3.2403],[759,-5.9166],[770,2.0843],[771,1.4565],[772,0.6644],[773,-0.2326],[774,-1.2079],[775,-2.1868],[776,-3.1393],[777,-4.008],[778,-4.7403],[779,-5.2928],[780,-5.8068],[781,-6.1955],[782,-6.4896],[783,-6.7209],[784,-6.8867],[785,-7.0154],[786,-7.0923],[787,-7.1346],[788,-7.1343],[789,-7.0976],[790,-7.0203],[791,-6.9074],[792,-6.7751],[793,-6.6231],[795,-6.2867],[796,-6.1076],[797,-5.9085],[799,-5.4757],[801,-5.011],[803,-4.5686],[837,-1.2578],[887,-6.2867],[888,-6.1076],[889,-5.9085],[891,-5.4757],[893,-5.011],[895,-4.5686],[929,-1.2578]]
       }
       head_ik_ctrl.Scale.X {type=MovieSceneScriptingFloatChannel, default=1, const [1,751]}
       head_ik_ctrl.Scale.Y {type=MovieSceneScriptingFloatChannel, default=1, const [1,751]}
       head_ik_ctrl.Scale.Z {type=MovieSceneScriptingFloatChannel, default=1, const [1,751]}
       chest_ctrl.Location.X {
        type=MovieSceneScriptingFloatChannel
        default=0
        thinned_from=751
        keys [[-2,-0.6631],[-1,-0.747],[0,-0.7952],[1,-0.8344],[3,-0.9067],[5,-0.9749],[7,-1.0341],[8,-1.0674],[9,-1.1128],[10,-1.1743],[11,-1.2481],[12,-1.33],[13,-1.4199],[15,-1.6197],[16,-1.729],[17,-1.8281],[18,-1.9348],[20,-2.1268],[21,-2.2171],[23,-2.3579],[25,-2.4525],[26,-2.4851],[28,-2.5387],[29,-2.5602],[30,-2.5726],[33,-2.5865],[37,-2.5809],[46,-2.5197],[52,-2.4784],[56,-2.4381],[60,-2.3892],[61,-2.3722],[64,-2.3356],[70,-2.254],[77,-2.1748],[82,-2.127],[83,-2.1224],[85,-2.1299],[87,-2.1597],[89,-2.2062],[90,-2.2328],[92,-2.272],[93,-2.2799],[94,-2.2774],[95,-2.2629],[96,-2.2272],[97,-2.1684],[98,-2.0937],[99,-2.004],[100,-1.894],[101,-1.7783],[103,-1.5099],[105,-1.2074],[106,-1.053],[107,-0.8919],[108,-0.7484],[109,-0.5967],[110,-0.45],[111,-0.3154],[112,-0.1828],[113,-0.0552],[114,0.0577],[115,0.1671],[116,0.2733],[117,0.3601],[118,0.4378],[119,0.4997],[120,0.553],[121,0.5863],[122,0.6143],[123,0.6404],[124,0.6453],[125,0.6401],[126,0.6313],[127,0.6165],[130,0.5649],[131,0.5511],[132,0.5386],[134,0.5214],[136,0.5107],[137,0.5063],[138,0.504],[139,0.5004],[142,0.4957],[145,0.495],[158,0.4973],[159,0.4993],[160,0.498],[162,0.5022],[163,0.5057],[164,0.5081],[165,0.5118],[170,0.5244],[171,0.526],[172,0.5249],[174,0.5204],[175,0.5137],[176,0.5053],[177,0.489],[178,0.4751],[179,0.4643],[180,0.4503],[181,0.4351],[182,0.4242],[183,0.4156],[184,0.4084],[186,0.4006],[188,0.3858],[189,0.3826],[190,0.3775],[191,0.3756],[192,0.3728],[195,0.3673],[198,0.3671],[200,0.3699],[202,0.3758],[203,0.3823],[204,0.3859],[205,0.3921],[207,0.4096],[208,0.4176],[209,0.4321],[210,0.442],[211,0.4552],[212,0.4705],[213,0.4909],[214,0.5084],[215,0.5247],[216,0.5384],[217,0.5534],[218,0.5705],[219,0.5926],[220,0.6245],[221,0.6673],[222,0.7189],[223,0.7781],[224,0.8398],[225,0.9044],[226,0.9648],[227,1.0227],[228,1.0739],[229,1.1126],[230,1.1428],[231,1.1528],[232,1.1542],[233,1.1429],[234,1.1274],[236,1.0874],[238,1.0395],[240,0.9815],[244,0.8479],[245,0.8178],[246,0.7853],[248,0.7256],[250,0.6566],[251,0.6153],[252,0.5706],[253,0.5223],[254,0.4714],[255,0.4165],[256,0.3588],[257,0.2959],[258,0.2356],[259,0.1727],[260,0.108],[261,0.0526],[262,0.0017],[263,-0.0399],[264,-0.0795],[265,-0.1058],[266,-0.126],[267,-0.1381],[268,-0.1467],[269,-0.1572],[270,-0.1649],[271,-0.1709],[273,-0.189],[274,-0.1984],[275,-0.2108],[277,-0.2272],[278,-0.2363],[279,-0.2443],[280,-0.2532],[281,-0.2693],[282,-0.2811],[284,-0.3003],[285,-0.3091],[286,-0.3213],[287,-0.3305],[288,-0.3404],[289,-0.3552],[291,-0.3765],[292,-0.3862],[293,-0.4043],[294,-0.4165],[295,-0.4347],[296,-0.4482],[298,-0.4862],[299,-0.5098],[300,-0.53],[302,-0.5812],[303,-0.6044],[304,-0.6319],[305,-0.6556],[306,-0.6811],[307,-0.7087],[308,-0.7309],[309,-0.758],[323,-0.6631],[324,-0.747],[325,-0.7952],[326,-0.8344],[328,-0.9067],[330,-0.9749],[332,-1.0341],[333,-1.0674],[334,-1.1128],[335,-1.1743],[336,-1.2481],[337,-1.33],[338,-1.4199],[340,-1.6197],[341,-1.729],[342,-1.8281],[343,-1.9348],[345,-2.1268],[346,-2.2171],[348,-2.3579],[350,-2.4525],[351,-2.4851],[353,-2.5387],[354,-2.5602],[355,-2.5726],[358,-2.5865],[362,-2.5809],[371,-2.5197],[377,-2.4784],[381,-2.4381],[385,-2.3892],[386,-2.3722],[389,-2.3356],[395,-2.254],[402,-2.1748],[407,-2.127],[408,-2.1224],[410,-2.1299],[412,-2.1597],[414,-2.2062],[415,-2.2328],[417,-2.272],[418,-2.2799],[419,-2.2774],[420,-2.2629],[421,-2.2272],[422,-2.1684],[423,-2.0937],[424,-2.004],[425,-1.894],[426,-1.7783],[428,-1.5099],[430,-1.2074],[431,-1.053],[432,-0.8919],[433,-0.7484],[434,-0.5967],[435,-0.45],[436,-0.3154],[437,-0.1828],[438,-0.0552],[439,0.0577],[440,0.1671],[441,0.2733],[442,0.3601],[443,0.4378],[444,0.4997],[445,0.553],[446,0.5863],[447,0.6143],[448,0.6404],[449,0.6453],[450,0.6401],[451,0.6313],[452,0.6165],[455,0.5649],[456,0.5511],[457,0.5386],[459,0.5214],[461,0.5107],[462,0.5063],[463,0.504],[464,0.5004],[467,0.4957],[470,0.495],[483,0.4973],[484,0.4993],[485,0.498],[487,0.5022],[488,0.5057],[489,0.5081],[490,0.5118],[495,0.5244],[496,0.526],[497,0.5249],[499,0.5204],[500,0.5137],[501,0.5053],[502,0.489],[503,0.4751],[504,0.4643],[505,0.4503],[506,0.4351],[507,0.4242],[508,0.4156],[509,0.4084],[511,0.4006],[513,0.3858],[514,0.3826],[515,0.3775],[516,0.3756],[517,0.3728],[520,0.3673],[523,0.3671],[525,0.3699],[527,0.3758],[528,0.3823],[529,0.3859],[530,0.3921],[532,0.4096],[533,0.4176],[534,0.4321],[535,0.442],[536,0.4552],[537,0.4705],[538,0.4909],[539,0.5084],[540,0.5247],[541,0.5384],[542,0.5534],[543,0.5705],[544,0.5926],[545,0.6245],[546,0.6673],[547,0.7189],[548,0.7781],[549,0.8398],[550,0.9044],[551,0.9648],[552,1.0227],[553,1.0739],[554,1.1126],[555,1.1428],[556,1.1528],[557,1.1542],[558,1.1429],[559,1.1274],[561,1.0874],[563,1.0395],[565,0.9815],[569,0.8479],[570,0.8178],[571,0.7853],[573,0.7256],[575,0.6566],[576,0.6153],[577,0.5706],[578,0.5223],[579,0.4714],[580,0.4165],[581,0.3588],[582,0.2959],[583,0.2356],[584,0.1727],[585,0.108],[586,0.0526],[587,0.0017],[588,-0.0399],[589,-0.0795],[590,-0.1058],[591,-0.126],[592,-0.1381],[593,-0.1467],[594,-0.1572],[595,-0.1649],[596,-0.1709],[598,-0.189],[599,-0.1984],[600,-0.2108],[602,-0.2272],[603,-0.2363],[604,-0.2443],[605,-0.2532],[606,-0.2693],[607,-0.2811],[609,-0.3003],[610,-0.3091],[611,-0.3213],[612,-0.3305],[613,-0.3404],[614,-0.3552],[616,-0.3765],[617,-0.3862],[618,-0.4043],[619,-0.4165],[620,-0.4347],[621,-0.4482],[623,-0.4862],[624,-0.5098],[625,-0.53],[627,-0.5812],[628,-0.6044],[629,-0.6319],[630,-0.6556],[631,-0.6811],[632,-0.7087],[633,-0.7309],[634,-0.758],[676,-0.6631],[677,-0.747],[678,-0.7952],[679,-0.8344],[681,-0.9067],[683,-0.9749],[685,-1.0341],[686,-1.0674],[687,-1.1128],[688,-1.1743],[689,-1.2481],[690,-1.33],[691,-1.4199],[693,-1.6197],[694,-1.729],[695,-1.8281],[696,-1.9348],[698,-2.1268],[699,-2.2171],[701,-2.3579],[703,-2.4525],[704,-2.4851],[706,-2.5387],[707,-2.5602],[708,-2.5726],[711,-2.5865],[715,-2.5809],[724,-2.5197],[730,-2.4784],[734,-2.4381],[738,-2.3892],[739,-2.3722],[742,-2.3356],[748,-2.254],[755,-2.1748],[759,-2.136],[770,-0.3091],[771,-0.3213],[772,-0.3305],[773,-0.3404],[774,-0.3552],[776,-0.3765],[777,-0.3862],[778,-0.4043],[779,-0.4165],[780,-0.4347],[781,-0.4482],[783,-0.4862],[784,-0.5098],[785,-0.53],[787,-0.5812],[788,-0.6044],[789,-0.6319],[790,-0.6556],[791,-0.6811],[792,-0.7087],[793,-0.7309],[794,-0.758],[795,-0.7801],[796,-0.805],[797,-0.826],[798,-0.85],[802,-0.921],[803,-0.9358],[837,-0.6631],[887,-0.7801],[888,-0.805],[889,-0.826],[890,-0.85],[894,-0.921],[895,-0.9358],[929,-0.6631]]
       }
       chest_ctrl.Location.Y {
        type=MovieSceneScriptingFloatChannel
        default=0
        thinned_from=751
        keys [[-2,-0.0278],[-1,-0.0071],[0,-0.0025],[2,-0.004],[3,-0.0028],[4,-0.0028],[5,-0.0068],[6,-0.0087],[7,-0.0118],[8,-0.0201],[9,-0.031],[10,-0.0455],[11,-0.0626],[12,-0.0833],[13,-0.1085],[14,-0.1363],[15,-0.1658],[16,-0.1981],[17,-0.2321],[20,-0.3382],[21,-0.3707],[22,-0.3996],[23,-0.4271],[24,-0.4493],[26,-0.4887],[28,-0.5204],[29,-0.5327],[31,-0.5516],[32,-0.559],[33,-0.5631],[34,-0.562],[36,-0.5651],[38,-0.5639],[39,-0.5547],[40,-0.5528],[42,-0.5454],[43,-0.5397],[44,-0.5289],[45,-0.5241],[47,-0.5131],[48,-0.5045],[49,-0.4941],[50,-0.4848],[51,-0.4728],[52,-0.4623],[54,-0.4372],[55,-0.4182],[56,-0.4033],[58,-0.3713],[59,-0.3546],[61,-0.3191],[62,-0.2979],[63,-0.2809],[64,-0.2616],[65,-0.2466],[67,-0.2103],[68,-0.1925],[69,-0.177],[71,-0.1504],[72,-0.1374],[73,-0.1256],[74,-0.1148],[75,-0.107],[76,-0.0997],[77,-0.0913],[78,-0.0837],[79,-0.0758],[80,-0.0689],[81,-0.0613],[82,-0.0524],[83,-0.0424],[84,-0.0276],[85,-0.0056],[86,0.0251],[87,0.0639],[88,0.1128],[89,0.1651],[90,0.2274],[91,0.2915],[92,0.3583],[93,0.4273],[94,0.4952],[96,0.6208],[97,0.6797],[98,0.7336],[99,0.785],[100,0.8328],[101,0.874],[102,0.9078],[103,0.9364],[104,0.9535],[105,0.9664],[106,0.9684],[107,0.9604],[108,0.9467],[109,0.9306],[110,0.9115],[111,0.8901],[113,0.8372],[114,0.8045],[115,0.7584],[116,0.7065],[117,0.6466],[118,0.582],[119,0.5128],[121,0.3661],[122,0.2892],[123,0.2162],[124,0.1413],[125,0.0734],[126,0.0057],[127,-0.0559],[128,-0.1126],[129,-0.1631],[130,-0.209],[131,-0.25],[132,-0.2853],[133,-0.3118],[134,-0.3349],[135,-0.3499],[136,-0.3629],[137,-0.3687],[139,-0.3649],[140,-0.3575],[141,-0.3483],[142,-0.337],[144,-0.3107],[146,-0.275],[147,-0.2559],[148,-0.2393],[149,-0.2207],[150,-0.2041],[151,-0.1851],[152,-0.1699],[153,-0.1519],[154,-0.1395],[155,-0.1241],[156,-0.1103],[158,-0.0904],[159,-0.0786],[160,-0.0711],[161,-0.0611],[162,-0.0596],[163,-0.0565],[165,-0.056],[166,-0.0571],[167,-0.0624],[168,-0.0716],[169,-0.0772],[170,-0.0871],[171,-0.0983],[172,-0.114],[173,-0.1274],[174,-0.1416],[175,-0.1552],[176,-0.1709],[177,-0.1835],[178,-0.1992],[179,-0.2112],[180,-0.224],[181,-0.2392],[182,-0.2445],[183,-0.2602],[184,-0.2703],[185,-0.2823],[186,-0.2909],[187,-0.3017],[188,-0.3091],[189,-0.3138],[191,-0.3254],[192,-0.3272],[193,-0.3309],[195,-0.3348],[196,-0.3301],[197,-0.3311],[199,-0.3311],[200,-0.3298],[201,-0.3254],[202,-0.3244],[203,-0.3223],[204,-0.3221],[205,-0.3139],[206,-0.3122],[207,-0.3098],[208,-0.3048],[209,-0.3021],[210,-0.2936],[211,-0.2891],[213,-0.2721],[214,-0.2685],[215,-0.268],[216,-0.2635],[217,-0.2621],[218,-0.2624],[219,-0.2725],[221,-0.275],[222,-0.2791],[223,-0.2761],[225,-0.2578],[226,-0.2408],[227,-0.2213],[229,-0.1597],[230,-0.1213],[231,-0.0751],[232,-0.0213],[233,0.0308],[234,0.0923],[235,0.1486],[236,0.2155],[237,0.2792],[238,0.3553],[239,0.4208],[241,0.5707],[242,0.6428],[244,0.7951],[245,0.8661],[246,0.9427],[248,1.0821],[249,1.1443],[251,1.2623],[252,1.315],[253,1.3636],[254,1.4085],[257,1.5228],[259,1.5903],[262,1.6821],[264,1.732],[267,1.7935],[270,1.8363],[273,1.855],[274,1.8553],[276,1.8511],[278,1.8642],[282,1.8693],[284,1.8802],[290,1.9081],[298,1.9828],[309,2.1093],[323,-0.0278],[324,-0.0071],[325,-0.0025],[327,-0.004],[328,-0.0028],[329,-0.0028],[330,-0.0068],[331,-0.0087],[332,-0.0118],[333,-0.0201],[334,-0.031],[335,-0.0455],[336,-0.0626],[337,-0.0833],[338,-0.1085],[339,-0.1363],[340,-0.1658],[341,-0.1981],[342,-0.2321],[345,-0.3382],[346,-0.3707],[347,-0.3996],[348,-0.4271],[349,-0.4493],[351,-0.4887],[353,-0.5204],[354,-0.5327],[356,-0.5516],[357,-0.559],[358,-0.5631],[359,-0.562],[361,-0.5651],[363,-0.5639],[364,-0.5547],[365,-0.5528],[367,-0.5454],[368,-0.5397],[369,-0.5289],[370,-0.5241],[372,-0.5131],[373,-0.5045],[374,-0.4941],[375,-0.4848],[376,-0.4728],[377,-0.4623],[379,-0.4372],[380,-0.4182],[381,-0.4033],[383,-0.3713],[384,-0.3546],[386,-0.3191],[387,-0.2979],[388,-0.2809],[389,-0.2616],[390,-0.2466],[392,-0.2103],[393,-0.1925],[394,-0.177],[396,-0.1504],[397,-0.1374],[398,-0.1256],[399,-0.1148],[400,-0.107],[401,-0.0997],[402,-0.0913],[403,-0.0837],[404,-0.0758],[405,-0.0689],[406,-0.0613],[407,-0.0524],[408,-0.0424],[409,-0.0276],[410,-0.0056],[411,0.0251],[412,0.0639],[413,0.1128],[414,0.1651],[415,0.2274],[416,0.2915],[417,0.3583],[418,0.4273],[419,0.4952],[421,0.6208],[422,0.6797],[423,0.7336],[424,0.785],[425,0.8328],[426,0.874],[427,0.9078],[428,0.9364],[429,0.9535],[430,0.9664],[431,0.9684],[432,0.9604],[433,0.9467],[434,0.9306],[435,0.9115],[436,0.8901],[438,0.8372],[439,0.8045],[440,0.7584],[441,0.7065],[442,0.6466],[443,0.582],[444,0.5128],[446,0.3661],[447,0.2892],[448,0.2162],[449,0.1413],[450,0.0734],[451,0.0057],[452,-0.0559],[453,-0.1126],[454,-0.1631],[455,-0.209],[456,-0.25],[457,-0.2853],[458,-0.3118],[459,-0.3349],[460,-0.3499],[461,-0.3629],[462,-0.3687],[464,-0.3649],[465,-0.3575],[466,-0.3483],[467,-0.337],[469,-0.3107],[471,-0.275],[472,-0.2559],[473,-0.2393],[474,-0.2207],[475,-0.2041],[476,-0.1851],[477,-0.1699],[478,-0.1519],[479,-0.1395],[480,-0.1241],[481,-0.1103],[483,-0.0904],[484,-0.0786],[485,-0.0711],[486,-0.0611],[487,-0.0596],[488,-0.0565],[490,-0.056],[491,-0.0571],[492,-0.0624],[493,-0.0716],[494,-0.0772],[495,-0.0871],[496,-0.0983],[497,-0.114],[498,-0.1274],[499,-0.1416],[500,-0.1552],[501,-0.1709],[502,-0.1835],[503,-0.1992],[504,-0.2112],[505,-0.224],[506,-0.2392],[507,-0.2445],[508,-0.2602],[509,-0.2703],[510,-0.2823],[511,-0.2909],[512,-0.3017],[513,-0.3091],[514,-0.3138],[516,-0.3254],[517,-0.3272],[518,-0.3309],[520,-0.3348],[521,-0.3301],[522,-0.3311],[524,-0.3311],[525,-0.3298],[526,-0.3254],[527,-0.3244],[528,-0.3223],[529,-0.3221],[530,-0.3139],[531,-0.3122],[532,-0.3098],[533,-0.3048],[534,-0.3021],[535,-0.2936],[536,-0.2891],[538,-0.2721],[539,-0.2685],[540,-0.268],[541,-0.2635],[542,-0.2621],[543,-0.2624],[544,-0.2725],[546,-0.275],[547,-0.2791],[548,-0.2761],[550,-0.2578],[551,-0.2408],[552,-0.2213],[554,-0.1597],[555,-0.1213],[556,-0.0751],[557,-0.0213],[558,0.0308],[559,0.0923],[560,0.1486],[561,0.2155],[562,0.2792],[563,0.3553],[564,0.4208],[566,0.5707],[567,0.6428],[569,0.7951],[570,0.8661],[571,0.9427],[573,1.0821],[574,1.1443],[576,1.2623],[577,1.315],[578,1.3636],[579,1.4085],[582,1.5228],[584,1.5903],[587,1.6821],[589,1.732],[592,1.7935],[595,1.8363],[598,1.855],[599,1.8553],[601,1.8511],[603,1.8642],[607,1.8693],[609,1.8802],[615,1.9081],[623,1.9828],[634,2.1093],[676,-0.0278],[677,-0.0071],[678,-0.0025],[680,-0.004],[681,-0.0028],[682,-0.0028],[683,-0.0068],[684,-0.0087],[685,-0.0118],[686,-0.0201],[687,-0.031],[688,-0.0455],[689,-0.0626],[690,-0.0833],[691,-0.1085],[692,-0.1363],[693,-0.1658],[694,-0.1981],[695,-0.2321],[698,-0.3382],[699,-0.3707],[700,-0.3996],[701,-0.4271],[702,-0.4493],[704,-0.4887],[706,-0.5204],[707,-0.5327],[709,-0.5516],[710,-0.559],[711,-0.5631],[712,-0.562],[714,-0.5651],[716,-0.5639],[717,-0.5547],[718,-0.5528],[720,-0.5454],[721,-0.5397],[722,-0.5289],[723,-0.5241],[725,-0.5131],[726,-0.5045],[727,-0.4941],[728,-0.4848],[729,-0.4728],[730,-0.4623],[732,-0.4372],[733,-0.4182],[734,-0.4033],[736,-0.3713],[737,-0.3546],[739,-0.3191],[740,-0.2979],[741,-0.2809],[742,-0.2616],[743,-0.2466],[745,-0.2103],[746,-0.1925],[747,-0.177],[749,-0.1504],[750,-0.1374],[751,-0.1256],[752,-0.1148],[753,-0.107],[754,-0.0997],[755,-0.0913],[757,-0.0758],[759,-0.0613],[770,1.8827],[775,1.9081],[783,1.9828],[796,2.129],[802,2.1762],[803,2.1804],[837,-0.0278],[887,2.1186],[894,2.1762],[895,2.1804],[929,-0.0278]]
       }
       chest_ctrl.Location.Z {
        type=MovieSceneScriptingFloatChannel
        default=0
        thinned_from=751
        keys [[-2,-0.0084],[-1,-0.0111],[1,-0.0153],[2,-0.0186],[4,-0.023],[5,-0.0256],[6,-0.0268],[7,-0.029],[8,-0.032],[9,-0.0337],[10,-0.0383],[11,-0.0441],[12,-0.0494],[13,-0.0583],[14,-0.0665],[15,-0.077],[16,-0.0908],[17,-0.0972],[18,-0.1083],[19,-0.1189],[20,-0.1267],[21,-0.1417],[22,-0.1471],[23,-0.1592],[24,-0.1604],[25,-0.1688],[27,-0.1724],[28,-0.1786],[29,-0.175],[30,-0.1768],[31,-0.1791],[32,-0.1786],[34,-0.1781],[35,-0.1764],[37,-0.1742],[39,-0.171],[40,-0.1683],[44,-0.1608],[46,-0.1576],[47,-0.1552],[48,-0.1545],[50,-0.1495],[51,-0.1508],[52,-0.1526],[54,-0.146],[55,-0.1438],[56,-0.1402],[57,-0.1444],[58,-0.1404],[60,-0.1356],[61,-0.1378],[62,-0.135],[63,-0.131],[64,-0.1288],[65,-0.1322],[67,-0.128],[68,-0.1243],[69,-0.1303],[71,-0.1236],[72,-0.1223],[73,-0.1194],[74,-0.1184],[75,-0.1236],[76,-0.1213],[77,-0.1208],[78,-0.1187],[79,-0.1172],[80,-0.1152],[81,-0.1161],[82,-0.1149],[83,-0.114],[84,-0.1138],[85,-0.1167],[86,-0.1177],[87,-0.1205],[88,-0.1209],[89,-0.124],[90,-0.1248],[91,-0.1285],[92,-0.1315],[93,-0.13],[94,-0.1313],[95,-0.1295],[96,-0.1287],[97,-0.1217],[98,-0.1174],[99,-0.1099],[101,-0.0968],[102,-0.0872],[103,-0.0813],[104,-0.0709],[105,-0.0669],[106,-0.0587],[107,-0.0532],[108,-0.0502],[109,-0.0464],[110,-0.0432],[111,-0.0413],[112,-0.0391],[114,-0.0393],[115,-0.0373],[116,-0.0348],[117,-0.0369],[118,-0.0322],[119,-0.0346],[120,-0.0298],[121,-0.0257],[122,-0.0287],[123,-0.0245],[124,-0.0195],[125,-0.023],[126,-0.0189],[127,-0.0186],[128,-0.0171],[129,-0.0162],[130,-0.0149],[131,-0.0134],[133,-0.0095],[134,-0.0111],[135,-0.0099],[136,-0.0093],[137,-0.0098],[138,-0.0094],[139,-0.0114],[140,-0.0127],[141,-0.0124],[142,-0.0132],[143,-0.0126],[144,-0.0146],[145,-0.0143],[146,-0.0168],[147,-0.0164],[148,-0.0182],[149,-0.0177],[150,-0.0195],[151,-0.0193],[152,-0.0212],[153,-0.0211],[154,-0.0234],[155,-0.0227],[156,-0.0223],[157,-0.0249],[158,-0.0256],[159,-0.0255],[161,-0.0265],[162,-0.0273],[163,-0.0272],[164,-0.0286],[165,-0.0288],[166,-0.03],[167,-0.0302],[168,-0.0311],[170,-0.0305],[171,-0.0317],[172,-0.0324],[176,-0.0322],[178,-0.0315],[179,-0.0294],[180,-0.0293],[182,-0.0286],[183,-0.0285],[184,-0.0254],[185,-0.025],[186,-0.0253],[188,-0.0242],[189,-0.0221],[190,-0.0211],[191,-0.0215],[192,-0.0208],[193,-0.0194],[194,-0.0193],[195,-0.0189],[196,-0.0181],[197,-0.0176],[198,-0.0161],[199,-0.0162],[201,-0.0152],[203,-0.0153],[204,-0.0147],[205,-0.0138],[206,-0.0126],[207,-0.0126],[208,-0.0134],[209,-0.0134],[210,-0.013],[211,-0.0133],[212,-0.0129],[213,-0.0128],[214,-0.0133],[215,-0.0147],[218,-0.0159],[219,-0.0173],[220,-0.0184],[221,-0.0199],[222,-0.0237],[223,-0.0258],[224,-0.03],[225,-0.0348],[226,-0.0365],[227,-0.0424],[228,-0.0448],[229,-0.0455],[230,-0.047],[231,-0.0467],[232,-0.0473],[234,-0.0433],[235,-0.0417],[236,-0.0428],[237,-0.0389],[238,-0.04],[239,-0.0378],[240,-0.0423],[241,-0.0382],[242,-0.0436],[243,-0.0403],[244,-0.0465],[245,-0.0445],[246,-0.0472],[247,-0.0506],[248,-0.0499],[249,-0.0584],[250,-0.0582],[251,-0.0672],[252,-0.0673],[253,-0.0766],[254,-0.0771],[255,-0.0792],[256,-0.0856],[257,-0.0882],[258,-0.091],[259,-0.0979],[260,-0.101],[261,-0.1028],[262,-0.1062],[263,-0.1119],[264,-0.1151],[266,-0.1179],[268,-0.1197],[270,-0.1208],[271,-0.1201],[272,-0.1198],[273,-0.1206],[274,-0.1186],[275,-0.1186],[276,-0.118],[277,-0.1127],[278,-0.1121],[279,-0.1122],[280,-0.1117],[284,-0.113],[285,-0.1118],[286,-0.1122],[287,-0.1149],[288,-0.1159],[290,-0.1171],[291,-0.1184],[292,-0.1201],[293,-0.1226],[294,-0.1241],[295,-0.1277],[296,-0.1294],[297,-0.1333],[298,-0.135],[299,-0.1398],[301,-0.1461],[302,-0.151],[303,-0.1539],[305,-0.1624],[306,-0.1661],[307,-0.1711],[308,-0.1739],[309,-0.1785],[323,-0.0084],[324,-0.0111],[326,-0.0153],[327,-0.0186],[329,-0.023],[330,-0.0256],[331,-0.0268],[332,-0.029],[333,-0.032],[334,-0.0337],[335,-0.0383],[336,-0.0441],[337,-0.0494],[338,-0.0583],[339,-0.0665],[340,-0.077],[341,-0.0908],[342,-0.0972],[343,-0.1083],[344,-0.1189],[345,-0.1267],[346,-0.1417],[347,-0.1471],[348,-0.1592],[349,-0.1604],[350,-0.1688],[352,-0.1724],[353,-0.1786],[354,-0.175],[355,-0.1768],[356,-0.1791],[357,-0.1786],[359,-0.1781],[360,-0.1764],[362,-0.1742],[364,-0.171],[365,-0.1683],[369,-0.1608],[371,-0.1576],[372,-0.1552],[373,-0.1545],[375,-0.1495],[376,-0.1508],[377,-0.1526],[379,-0.146],[380,-0.1438],[381,-0.1402],[382,-0.1444],[383,-0.1404],[385,-0.1356],[386,-0.1378],[387,-0.135],[388,-0.131],[389,-0.1288],[390,-0.1322],[392,-0.128],[393,-0.1243],[394,-0.1303],[396,-0.1236],[397,-0.1223],[398,-0.1194],[399,-0.1184],[400,-0.1236],[401,-0.1213],[402,-0.1208],[403,-0.1187],[404,-0.1172],[405,-0.1152],[406,-0.1161],[407,-0.1149],[408,-0.114],[409,-0.1138],[410,-0.1167],[411,-0.1177],[412,-0.1205],[413,-0.1209],[414,-0.124],[415,-0.1248],[416,-0.1285],[417,-0.1315],[418,-0.13],[419,-0.1313],[420,-0.1295],[421,-0.1287],[422,-0.1217],[423,-0.1174],[424,-0.1099],[426,-0.0968],[427,-0.0872],[428,-0.0813],[429,-0.0709],[430,-0.0669],[431,-0.0587],[432,-0.0532],[433,-0.0502],[434,-0.0464],[435,-0.0432],[436,-0.0413],[437,-0.0391],[439,-0.0393],[440,-0.0373],[441,-0.0348],[442,-0.0369],[443,-0.0322],[444,-0.0346],[445,-0.0298],[446,-0.0257],[447,-0.0287],[448,-0.0245],[449,-0.0195],[450,-0.023],[451,-0.0189],[452,-0.0186],[453,-0.0171],[454,-0.0162],[455,-0.0149],[456,-0.0134],[458,-0.0095],[459,-0.0111],[460,-0.0099],[461,-0.0093],[462,-0.0098],[463,-0.0094],[464,-0.0114],[465,-0.0127],[466,-0.0124],[467,-0.0132],[468,-0.0126],[469,-0.0146],[470,-0.0143],[471,-0.0168],[472,-0.0164],[473,-0.0182],[474,-0.0177],[475,-0.0195],[476,-0.0193],[477,-0.0212],[478,-0.0211],[479,-0.0234],[480,-0.0227],[481,-0.0223],[482,-0.0249],[483,-0.0256],[484,-0.0255],[486,-0.0265],[487,-0.0273],[488,-0.0272],[489,-0.0286],[490,-0.0288],[491,-0.03],[492,-0.0302],[493,-0.0311],[495,-0.0305],[496,-0.0317],[497,-0.0324],[501,-0.0322],[503,-0.0315],[504,-0.0294],[505,-0.0293],[507,-0.0286],[508,-0.0285],[509,-0.0254],[510,-0.025],[511,-0.0253],[513,-0.0242],[514,-0.0221],[515,-0.0211],[516,-0.0215],[517,-0.0208],[518,-0.0194],[519,-0.0193],[520,-0.0189],[521,-0.0181],[522,-0.0176],[523,-0.0161],[524,-0.0162],[526,-0.0152],[528,-0.0153],[529,-0.0147],[530,-0.0138],[531,-0.0126],[532,-0.0126],[533,-0.0134],[534,-0.0134],[535,-0.013],[536,-0.0133],[537,-0.0129],[538,-0.0128],[539,-0.0133],[540,-0.0147],[543,-0.0159],[544,-0.0173],[545,-0.0184],[546,-0.0199],[547,-0.0237],[548,-0.0258],[549,-0.03],[550,-0.0348],[551,-0.0365],[552,-0.0424],[553,-0.0448],[554,-0.0455],[555,-0.047],[556,-0.0467],[557,-0.0473],[559,-0.0433],[560,-0.0417],[561,-0.0428],[562,-0.0389],[563,-0.04],[564,-0.0378],[565,-0.0423],[566,-0.0382],[567,-0.0436],[568,-0.0403],[569,-0.0465],[570,-0.0445],[571,-0.0472],[572,-0.0506],[573,-0.0499],[574,-0.0584],[575,-0.0582],[576,-0.0672],[577,-0.0673],[578,-0.0766],[579,-0.0771],[580,-0.0792],[581,-0.0856],[582,-0.0882],[583,-0.091],[584,-0.0979],[585,-0.101],[586,-0.1028],[587,-0.1062],[588,-0.1119],[589,-0.1151],[591,-0.1179],[593,-0.1197],[595,-0.1208],[596,-0.1201],[597,-0.1198],[598,-0.1206],[599,-0.1186],[600,-0.1186],[601,-0.118],[602,-0.1127],[603,-0.1121],[604,-0.1122],[605,-0.1117],[609,-0.113],[610,-0.1118],[611,-0.1122],[612,-0.1149],[613,-0.1159],[615,-0.1171],[616,-0.1184],[617,-0.1201],[618,-0.1226],[619,-0.1241],[620,-0.1277],[621,-0.1294],[622,-0.1333],[623,-0.135],[624,-0.1398],[626,-0.1461],[627,-0.151],[628,-0.1539],[630,-0.1624],[631,-0.1661],[632,-0.1711],[633,-0.1739],[634,-0.1785],[676,-0.0084],[677,-0.0111],[679,-0.0153],[680,-0.0186],[682,-0.023],[683,-0.0256],[684,-0.0268],[685,-0.029],[686,-0.032],[687,-0.0337],[688,-0.0383],[689,-0.0441],[690,-0.0494],[691,-0.0583],[692,-0.0665],[693,-0.077],[694,-0.0908],[695,-0.0972],[696,-0.1083],[697,-0.1189],[698,-0.1267],[699,-0.1417],[700,-0.1471],[701,-0.1592],[702,-0.1604],[703,-0.1688],[705,-0.1724],[706,-0.1786],[707,-0.175],[708,-0.1768],[709,-0.1791],[710,-0.1786],[712,-0.1781],[713,-0.1764],[715,-0.1742],[717,-0.171],[718,-0.1683],[722,-0.1608],[724,-0.1576],[725,-0.1552],[726,-0.1545],[728,-0.1495],[729,-0.1508],[730,-0.1526],[732,-0.146],[733,-0.1438],[734,-0.1402],[735,-0.1444],[736,-0.1404],[738,-0.1356],[739,-0.1378],[740,-0.135],[741,-0.131],[742,-0.1288],[743,-0.1322],[745,-0.128],[746,-0.1243],[747,-0.1303],[749,-0.1236],[750,-0.1223],[751,-0.1194],[752,-0.1184],[753,-0.1236],[754,-0.1213],[755,-0.1208],[757,-0.1172],[759,-0.1161],[770,-0.1118],[771,-0.1122],[772,-0.1149],[773,-0.1159],[775,-0.1171],[776,-0.1184],[777,-0.1201],[778,-0.1226],[779,-0.1241],[780,-0.1277],[781,-0.1294],[782,-0.1333],[783,-0.135],[784,-0.1398],[786,-0.1461],[787,-0.151],[788,-0.1539],[790,-0.1624],[791,-0.1661],[792,-0.1711],[793,-0.1739],[794,-0.1785],[795,-0.1814],[796,-0.1859],[799,-0.1948],[801,-0.2001],[802,-0.2032],[803,-0.2055],[837,-0.0084],[887,-0.1814],[888,-0.1859],[891,-0.1948],[893,-0.2001],[894,-0.2032],[895,-0.2055],[929,-0.0084]]
       }
       chest_ctrl.Rotation.X {
        type=MovieSceneScriptingFloatChannel
        default=0
        thinned_from=751
        keys [[-2,-2.2727],[-1,-2.1691],[0,-2.161],[1,-2.1623],[2,-2.1707],[3,-2.1873],[4,-2.2132],[5,-2.245],[6,-2.2861],[7,-2.3359],[8,-2.4025],[9,-2.4862],[10,-2.5871],[11,-2.7097],[12,-2.85],[13,-3.0109],[14,-3.1862],[15,-3.3739],[17,-3.7766],[19,-4.1904],[21,-4.5799],[22,-4.7584],[23,-4.9212],[24,-5.0655],[25,-5.1955],[26,-5.3119],[27,-5.4169],[29,-5.5828],[31,-5.6979],[33,-5.7596],[35,-5.7886],[37,-5.7947],[40,-5.7655],[47,-5.5897],[49,-5.5101],[52,-5.3486],[55,-5.1412],[58,-4.8945],[68,-3.9534],[70,-3.7909],[71,-3.7161],[73,-3.5855],[76,-3.4331],[82,-3.19],[84,-3.0935],[85,-3.0247],[86,-2.9141],[87,-2.7682],[88,-2.5962],[89,-2.4044],[90,-2.1944],[91,-1.9708],[92,-1.7333],[93,-1.4854],[94,-1.2312],[95,-0.9704],[96,-0.6947],[97,-0.395],[98,-0.0804],[99,0.243],[101,0.9017],[102,1.2222],[103,1.525],[104,1.8037],[105,2.049],[106,2.2528],[107,2.4094],[108,2.5274],[109,2.6212],[110,2.6917],[111,2.7345],[112,2.751],[113,2.7424],[114,2.6887],[115,2.5749],[116,2.4064],[117,2.1875],[118,1.9292],[119,1.632],[120,1.306],[121,0.9567],[122,0.5943],[123,0.2226],[124,-0.1545],[125,-0.5287],[126,-0.8958],[127,-1.2469],[128,-1.5765],[129,-1.8823],[130,-2.1582],[131,-2.4063],[132,-2.6218],[133,-2.8037],[134,-2.9519],[135,-3.0605],[136,-3.1305],[137,-3.1594],[138,-3.1568],[139,-3.1355],[140,-3.0982],[141,-3.0467],[142,-2.9808],[143,-2.9029],[144,-2.8122],[145,-2.7143],[150,-2.1822],[152,-1.9851],[153,-1.8917],[155,-1.7313],[156,-1.6581],[157,-1.5909],[158,-1.5305],[160,-1.4186],[161,-1.3729],[162,-1.3391],[163,-1.3203],[164,-1.3052],[165,-1.3078],[167,-1.334],[168,-1.3647],[169,-1.405],[170,-1.4556],[172,-1.5831],[173,-1.6567],[174,-1.7355],[177,-1.995],[179,-2.1724],[182,-2.4149],[184,-2.5639],[186,-2.6922],[187,-2.7501],[188,-2.8005],[189,-2.8452],[190,-2.8834],[191,-2.9153],[193,-2.9551],[195,-2.9759],[197,-2.9807],[199,-2.9736],[202,-2.9438],[206,-2.8767],[208,-2.8274],[210,-2.7594],[212,-2.6734],[213,-2.6254],[214,-2.5831],[215,-2.5505],[216,-2.5257],[217,-2.5083],[218,-2.4989],[220,-2.4917],[221,-2.4533],[222,-2.3905],[223,-2.3073],[224,-2.2083],[225,-2.0937],[226,-1.9645],[227,-1.8193],[228,-1.6606],[229,-1.4882],[230,-1.3019],[231,-1.1038],[232,-0.8884],[233,-0.6512],[234,-0.3991],[235,-0.132],[236,0.1478],[237,0.4406],[238,0.7448],[239,1.0618],[240,1.3931],[241,1.7368],[246,3.477],[247,3.8135],[248,4.1317],[249,4.4356],[250,4.7209],[251,4.9895],[252,5.2393],[253,5.4716],[254,5.6897],[256,6.0747],[258,6.4155],[260,6.7178],[262,6.9883],[264,7.2253],[266,7.4277],[268,7.594],[270,7.7171],[272,7.797],[274,7.8192],[278,7.8265],[287,7.9924],[295,8.32],[308,8.9921],[309,9.0317],[323,-2.2727],[324,-2.1691],[325,-2.161],[326,-2.1623],[327,-2.1707],[328,-2.1873],[329,-2.2132],[330,-2.245],[331,-2.2861],[332,-2.3359],[333,-2.4025],[334,-2.4862],[335,-2.5871],[336,-2.7097],[337,-2.85],[338,-3.0109],[339,-3.1862],[340,-3.3739],[342,-3.7766],[344,-4.1904],[346,-4.5799],[347,-4.7584],[348,-4.9212],[349,-5.0655],[350,-5.1955],[351,-5.3119],[352,-5.4169],[354,-5.5828],[356,-5.6979],[358,-5.7596],[360,-5.7886],[362,-5.7947],[365,-5.7655],[372,-5.5897],[374,-5.5101],[377,-5.3486],[380,-5.1412],[383,-4.8945],[393,-3.9534],[395,-3.7909],[396,-3.7161],[398,-3.5855],[401,-3.4331],[407,-3.19],[409,-3.0935],[410,-3.0247],[411,-2.9141],[412,-2.7682],[413,-2.5962],[414,-2.4044],[415,-2.1944],[416,-1.9708],[417,-1.7333],[418,-1.4854],[419,-1.2312],[420,-0.9704],[421,-0.6947],[422,-0.395],[423,-0.0804],[424,0.243],[426,0.9017],[427,1.2222],[428,1.525],[429,1.8037],[430,2.049],[431,2.2528],[432,2.4094],[433,2.5274],[434,2.6212],[435,2.6917],[436,2.7345],[437,2.751],[438,2.7424],[439,2.6887],[440,2.5749],[441,2.4064],[442,2.1875],[443,1.9292],[444,1.632],[445,1.306],[446,0.9567],[447,0.5943],[448,0.2226],[449,-0.1545],[450,-0.5287],[451,-0.8958],[452,-1.2469],[453,-1.5765],[454,-1.8823],[455,-2.1582],[456,-2.4063],[457,-2.6218],[458,-2.8037],[459,-2.9519],[460,-3.0605],[461,-3.1305],[462,-3.1594],[463,-3.1568],[464,-3.1355],[465,-3.0982],[466,-3.0467],[467,-2.9808],[468,-2.9029],[469,-2.8122],[470,-2.7143],[475,-2.1822],[477,-1.9851],[478,-1.8917],[480,-1.7313],[481,-1.6581],[482,-1.5909],[483,-1.5305],[485,-1.4186],[486,-1.3729],[487,-1.3391],[488,-1.3203],[489,-1.3052],[490,-1.3078],[492,-1.334],[493,-1.3647],[494,-1.405],[495,-1.4556],[497,-1.5831],[498,-1.6567],[499,-1.7355],[502,-1.995],[504,-2.1724],[507,-2.4149],[509,-2.5639],[511,-2.6922],[512,-2.7501],[513,-2.8005],[514,-2.8452],[515,-2.8834],[516,-2.9153],[518,-2.9551],[520,-2.9759],[522,-2.9807],[524,-2.9736],[527,-2.9438],[531,-2.8767],[533,-2.8274],[535,-2.7594],[537,-2.6734],[538,-2.6254],[539,-2.5831],[540,-2.5505],[541,-2.5257],[542,-2.5083],[543,-2.4989],[545,-2.4917],[546,-2.4533],[547,-2.3905],[548,-2.3073],[549,-2.2083],[550,-2.0937],[551,-1.9645],[552,-1.8193],[553,-1.6606],[554,-1.4882],[555,-1.3019],[556,-1.1038],[557,-0.8884],[558,-0.6512],[559,-0.3991],[560,-0.132],[561,0.1478],[562,0.4406],[563,0.7448],[564,1.0618],[565,1.3931],[566,1.7368],[571,3.477],[572,3.8135],[573,4.1317],[574,4.4356],[575,4.7209],[576,4.9895],[577,5.2393],[578,5.4716],[579,5.6897],[581,6.0747],[583,6.4155],[585,6.7178],[587,6.9883],[589,7.2253],[591,7.4277],[593,7.594],[595,7.7171],[597,7.797],[599,7.8192],[603,7.8265],[612,7.9924],[620,8.32],[633,8.9921],[634,9.0317],[676,-2.2727],[677,-2.1691],[678,-2.161],[679,-2.1623],[680,-2.1707],[681,-2.1873],[682,-2.2132],[683,-2.245],[684,-2.2861],[685,-2.3359],[686,-2.4025],[687,-2.4862],[688,-2.5871],[689,-2.7097],[690,-2.85],[691,-3.0109],[692,-3.1862],[693,-3.3739],[695,-3.7766],[697,-4.1904],[699,-4.5799],[700,-4.7584],[701,-4.9212],[702,-5.0655],[703,-5.1955],[704,-5.3119],[705,-5.4169],[707,-5.5828],[709,-5.6979],[711,-5.7596],[713,-5.7886],[715,-5.7947],[718,-5.7655],[725,-5.5897],[727,-5.5101],[730,-5.3486],[733,-5.1412],[736,-4.8945],[746,-3.9534],[748,-3.7909],[749,-3.7161],[751,-3.5855],[754,-3.4331],[759,-3.2329],[770,7.9456],[774,8.0551],[793,8.9921],[799,9.1785],[803,9.2522],[837,-2.2727],[887,9.0681],[895,9.2522],[929,-2.2727]]
       }
       chest_ctrl.Rotation.Y {
        type=MovieSceneScriptingFloatChannel
        default=0
        thinned_from=751
        keys [[-2,4.2758],[-1,4.6739],[0,4.8582],[1,5.0299],[3,5.3341],[4,5.4697],[6,5.7086],[7,5.8136],[8,5.9387],[9,6.1207],[10,6.3658],[11,6.6611],[12,6.9998],[13,7.375],[15,8.1956],[18,9.4641],[19,9.8626],[20,10.2375],[21,10.5793],[22,10.8764],[23,11.1238],[24,11.3109],[26,11.5795],[28,11.7632],[30,11.8693],[33,11.9149],[53,11.4116],[82,10.0933],[84,10.0542],[85,10.0662],[87,10.143],[91,10.429],[92,10.4842],[93,10.5078],[94,10.4893],[95,10.4149],[96,10.2686],[97,10.0386],[98,9.7365],[99,9.3755],[100,8.9535],[101,8.4776],[102,7.951],[103,7.3739],[104,6.7543],[105,6.1023],[108,4.0918],[109,3.4353],[110,2.7981],[111,2.1855],[112,1.6035],[113,1.0439],[114,0.5026],[115,-0.009],[116,-0.4838],[117,-0.9168],[118,-1.3023],[119,-1.6352],[120,-1.9129],[121,-2.1427],[122,-2.3353],[123,-2.4877],[124,-2.5992],[125,-2.6744],[126,-2.7198],[127,-2.7431],[131,-2.7581],[135,-2.7915],[136,-2.7927],[138,-2.7806],[141,-2.7368],[145,-2.6539],[153,-2.4631],[156,-2.4086],[160,-2.3632],[161,-2.3571],[162,-2.3564],[164,-2.37],[167,-2.4126],[171,-2.4834],[173,-2.4986],[174,-2.5008],[175,-2.4909],[176,-2.4711],[178,-2.4053],[180,-2.3291],[182,-2.2656],[183,-2.2413],[193,-2.1072],[197,-2.0716],[200,-2.0637],[202,-2.0711],[204,-2.0895],[206,-2.1214],[208,-2.1678],[209,-2.1958],[211,-2.2649],[212,-2.3098],[213,-2.3618],[214,-2.4075],[215,-2.4369],[216,-2.4494],[217,-2.4515],[218,-2.4609],[219,-2.49],[220,-2.5486],[221,-2.6462],[222,-2.7765],[223,-2.9295],[225,-3.263],[226,-3.4245],[227,-3.5697],[228,-3.6881],[229,-3.7697],[230,-3.8034],[231,-3.7788],[232,-3.7112],[233,-3.6182],[237,-3.1596],[239,-2.9138],[240,-2.7814],[242,-2.5045],[244,-2.2406],[245,-2.1175],[247,-1.8837],[248,-1.7609],[249,-1.6302],[250,-1.4788],[251,-1.3085],[252,-1.1184],[253,-0.9107],[254,-0.6867],[255,-0.4453],[256,-0.1877],[257,0.0796],[258,0.3615],[259,0.6457],[260,0.9224],[261,1.1821],[262,1.4088],[263,1.6026],[264,1.7605],[265,1.8818],[266,1.9732],[267,2.0249],[271,2.16],[274,2.2804],[275,2.3246],[279,2.4773],[286,2.7895],[287,2.8413],[290,2.9864],[292,3.1013],[295,3.2966],[296,3.3717],[297,3.4573],[299,3.6527],[301,3.8733],[309,4.8435],[323,4.2758],[324,4.6739],[325,4.8582],[326,5.0299],[328,5.3341],[329,5.4697],[331,5.7086],[332,5.8136],[333,5.9387],[334,6.1207],[335,6.3658],[336,6.6611],[337,6.9998],[338,7.375],[340,8.1956],[343,9.4641],[344,9.8626],[345,10.2375],[346,10.5793],[347,10.8764],[348,11.1238],[349,11.3109],[351,11.5795],[353,11.7632],[355,11.8693],[358,11.9149],[378,11.4116],[407,10.0933],[409,10.0542],[410,10.0662],[412,10.143],[416,10.429],[417,10.4842],[418,10.5078],[419,10.4893],[420,10.4149],[421,10.2686],[422,10.0386],[423,9.7365],[424,9.3755],[425,8.9535],[426,8.4776],[427,7.951],[428,7.3739],[429,6.7543],[430,6.1023],[433,4.0918],[434,3.4353],[435,2.7981],[436,2.1855],[437,1.6035],[438,1.0439],[439,0.5026],[440,-0.009],[441,-0.4838],[442,-0.9168],[443,-1.3023],[444,-1.6352],[445,-1.9129],[446,-2.1427],[447,-2.3353],[448,-2.4877],[449,-2.5992],[450,-2.6744],[451,-2.7198],[452,-2.7431],[456,-2.7581],[460,-2.7915],[461,-2.7927],[463,-2.7806],[466,-2.7368],[470,-2.6539],[478,-2.4631],[481,-2.4086],[485,-2.3632],[486,-2.3571],[487,-2.3564],[489,-2.37],[492,-2.4126],[496,-2.4834],[498,-2.4986],[499,-2.5008],[500,-2.4909],[501,-2.4711],[503,-2.4053],[505,-2.3291],[507,-2.2656],[508,-2.2413],[518,-2.1072],[522,-2.0716],[525,-2.0637],[527,-2.0711],[529,-2.0895],[531,-2.1214],[533,-2.1678],[534,-2.1958],[536,-2.2649],[537,-2.3098],[538,-2.3618],[539,-2.4075],[540,-2.4369],[541,-2.4494],[542,-2.4515],[543,-2.4609],[544,-2.49],[545,-2.5486],[546,-2.6462],[547,-2.7765],[548,-2.9295],[550,-3.263],[551,-3.4245],[552,-3.5697],[553,-3.6881],[554,-3.7697],[555,-3.8034],[556,-3.7788],[557,-3.7112],[558,-3.6182],[562,-3.1596],[564,-2.9138],[565,-2.7814],[567,-2.5045],[569,-2.2406],[570,-2.1175],[572,-1.8837],[573,-1.7609],[574,-1.6302],[575,-1.4788],[576,-1.3085],[577,-1.1184],[578,-0.9107],[579,-0.6867],[580,-0.4453],[581,-0.1877],[582,0.0796],[583,0.3615],[584,0.6457],[585,0.9224],[586,1.1821],[587,1.4088],[588,1.6026],[589,1.7605],[590,1.8818],[591,1.9732],[592,2.0249],[596,2.16],[599,2.2804],[600,2.3246],[604,2.4773],[611,2.7895],[612,2.8413],[615,2.9864],[617,3.1013],[620,3.2966],[621,3.3717],[622,3.4573],[624,3.6527],[626,3.8733],[634,4.8435],[676,4.2758],[677,4.6739],[678,4.8582],[679,5.0299],[681,5.3341],[682,5.4697],[684,5.7086],[685,5.8136],[686,5.9387],[687,6.1207],[688,6.3658],[689,6.6611],[690,6.9998],[691,7.375],[693,8.1956],[696,9.4641],[697,9.8626],[698,10.2375],[699,10.5793],[700,10.8764],[701,11.1238],[702,11.3109],[704,11.5795],[706,11.7632],[708,11.8693],[711,11.9149],[731,11.4116],[755,10.2971],[759,10.1325],[770,2.7437],[771,2.7895],[772,2.8413],[775,2.9864],[777,3.1013],[780,3.2966],[781,3.3717],[782,3.4573],[784,3.6527],[786,3.8733],[795,4.9627],[798,5.2914],[800,5.4803],[803,5.721],[837,4.2758],[887,4.9627],[890,5.2914],[892,5.4803],[895,5.721],[929,4.2758]]
       }
       chest_ctrl.Rotation.Z {
        type=MovieSceneScriptingFloatChannel
        default=0
        thinned_from=751
        keys [[-2,1.1133],[-1,1.1119],[0,1.0355],[1,0.9396],[2,0.8261],[3,0.6964],[4,0.5504],[5,0.3893],[6,0.2145],[7,0.025],[8,-0.1737],[9,-0.3815],[10,-0.5958],[11,-0.8147],[12,-1.0407],[14,-1.5039],[15,-1.7429],[16,-1.9856],[17,-2.2239],[19,-2.7089],[20,-2.9451],[22,-3.3931],[23,-3.6009],[24,-3.7942],[25,-3.9714],[26,-4.1247],[27,-4.2535],[28,-4.3577],[29,-4.438],[30,-4.5034],[31,-4.5493],[37,-4.7158],[42,-4.8123],[47,-4.8726],[54,-4.8877],[77,-4.6881],[81,-4.6822],[82,-4.7141],[83,-4.8035],[84,-4.9511],[85,-5.1812],[86,-5.4792],[87,-5.8299],[88,-6.2477],[89,-6.7351],[90,-7.2707],[92,-8.3979],[93,-8.9458],[94,-9.4534],[95,-9.8984],[96,-10.2564],[97,-10.5108],[98,-10.6793],[99,-10.7707],[100,-10.789],[101,-10.7351],[102,-10.6096],[103,-10.4143],[104,-10.1473],[105,-9.831],[107,-9.1288],[110,-8.015],[111,-7.6582],[114,-6.6721],[115,-6.3599],[116,-6.0658],[117,-5.8097],[118,-5.6069],[119,-5.4677],[120,-5.3995],[121,-5.3919],[122,-5.4238],[123,-5.5038],[124,-5.6367],[125,-5.8122],[126,-6.0211],[129,-6.7068],[130,-6.9146],[131,-7.1032],[132,-7.2667],[133,-7.4087],[134,-7.5305],[135,-7.6282],[136,-7.7092],[137,-7.7703],[139,-7.8516],[142,-7.8968],[145,-7.8692],[163,-7.5433],[170,-7.5029],[172,-7.5317],[174,-7.6084],[175,-7.668],[177,-7.857],[179,-8.0783],[181,-8.2648],[182,-8.3404],[183,-8.397],[187,-8.4696],[191,-8.4666],[195,-8.3688],[198,-8.234],[201,-8.029],[204,-7.7541],[206,-7.5255],[208,-7.2562],[210,-6.9432],[211,-6.771],[212,-6.5778],[213,-6.352],[214,-6.0885],[215,-5.7857],[216,-5.4512],[217,-5.0808],[218,-4.6535],[219,-4.1609],[220,-3.5856],[221,-2.932],[222,-2.2225],[223,-1.4776],[224,-0.6992],[225,0.0903],[226,0.884],[227,1.6689],[228,2.4326],[229,3.1547],[230,3.8248],[231,4.4292],[232,4.9583],[233,5.3938],[234,5.7589],[235,6.0712],[236,6.3406],[237,6.5624],[238,6.7423],[239,6.8826],[240,6.9852],[241,7.0517],[242,7.0849],[243,7.0904],[244,7.0676],[245,7.0187],[246,6.9477],[247,6.8557],[248,6.7451],[249,6.6186],[250,6.4779],[253,5.9954],[256,5.486],[258,5.1791],[259,5.0422],[260,4.9226],[261,4.8167],[262,4.7308],[263,4.6629],[264,4.6189],[265,4.5999],[266,4.6035],[267,4.6413],[270,4.7769],[274,4.9087],[290,5.2329],[292,5.2363],[294,5.2054],[295,5.1787],[296,5.1345],[297,5.069],[298,4.9842],[299,4.8843],[300,4.7674],[301,4.6356],[302,4.4929],[304,4.1796],[307,3.6512],[309,3.288],[323,1.1133],[324,1.1119],[325,1.0355],[326,0.9396],[327,0.8261],[328,0.6964],[329,0.5504],[330,0.3893],[331,0.2145],[332,0.025],[333,-0.1737],[334,-0.3815],[335,-0.5958],[336,-0.8147],[337,-1.0407],[339,-1.5039],[340,-1.7429],[341,-1.9856],[342,-2.2239],[344,-2.7089],[345,-2.9451],[347,-3.3931],[348,-3.6009],[349,-3.7942],[350,-3.9714],[351,-4.1247],[352,-4.2535],[353,-4.3577],[354,-4.438],[355,-4.5034],[356,-4.5493],[362,-4.7158],[367,-4.8123],[372,-4.8726],[379,-4.8877],[402,-4.6881],[406,-4.6822],[407,-4.7141],[408,-4.8035],[409,-4.9511],[410,-5.1812],[411,-5.4792],[412,-5.8299],[413,-6.2477],[414,-6.7351],[415,-7.2707],[417,-8.3979],[418,-8.9458],[419,-9.4534],[420,-9.8984],[421,-10.2564],[422,-10.5108],[423,-10.6793],[424,-10.7707],[425,-10.789],[426,-10.7351],[427,-10.6096],[428,-10.4143],[429,-10.1473],[430,-9.831],[432,-9.1288],[435,-8.015],[436,-7.6582],[439,-6.6721],[440,-6.3599],[441,-6.0658],[442,-5.8097],[443,-5.6069],[444,-5.4677],[445,-5.3995],[446,-5.3919],[447,-5.4238],[448,-5.5038],[449,-5.6367],[450,-5.8122],[451,-6.0211],[454,-6.7068],[455,-6.9146],[456,-7.1032],[457,-7.2667],[458,-7.4087],[459,-7.5305],[460,-7.6282],[461,-7.7092],[462,-7.7703],[464,-7.8516],[467,-7.8968],[470,-7.8692],[488,-7.5433],[495,-7.5029],[497,-7.5317],[499,-7.6084],[500,-7.668],[502,-7.857],[504,-8.0783],[506,-8.2648],[507,-8.3404],[508,-8.397],[512,-8.4696],[516,-8.4666],[520,-8.3688],[523,-8.234],[526,-8.029],[529,-7.7541],[531,-7.5255],[533,-7.2562],[535,-6.9432],[536,-6.771],[537,-6.5778],[538,-6.352],[539,-6.0885],[540,-5.7857],[541,-5.4512],[542,-5.0808],[543,-4.6535],[544,-4.1609],[545,-3.5856],[546,-2.932],[547,-2.2225],[548,-1.4776],[549,-0.6992],[550,0.0903],[551,0.884],[552,1.6689],[553,2.4326],[554,3.1547],[555,3.8248],[556,4.4292],[557,4.9583],[558,5.3938],[559,5.7589],[560,6.0712],[561,6.3406],[562,6.5624],[563,6.7423],[564,6.8826],[565,6.9852],[566,7.0517],[567,7.0849],[568,7.0904],[569,7.0676],[570,7.0187],[571,6.9477],[572,6.8557],[573,6.7451],[574,6.6186],[575,6.4779],[578,5.9954],[581,5.486],[583,5.1791],[584,5.0422],[585,4.9226],[586,4.8167],[587,4.7308],[588,4.6629],[589,4.6189],[590,4.5999],[591,4.6035],[592,4.6413],[595,4.7769],[599,4.9087],[615,5.2329],[617,5.2363],[619,5.2054],[620,5.1787],[621,5.1345],[622,5.069],[623,4.9842],[624,4.8843],[625,4.7674],[626,4.6356],[627,4.4929],[629,4.1796],[632,3.6512],[634,3.288],[676,1.1133],[677,1.1119],[678,1.0355],[679,0.9396],[680,0.8261],[681,0.6964],[682,0.5504],[683,0.3893],[684,0.2145],[685,0.025],[686,-0.1737],[687,-0.3815],[688,-0.5958],[689,-0.8147],[690,-1.0407],[692,-1.5039],[693,-1.7429],[694,-1.9856],[695,-2.2239],[697,-2.7089],[698,-2.9451],[700,-3.3931],[701,-3.6009],[702,-3.7942],[703,-3.9714],[704,-4.1247],[705,-4.2535],[706,-4.3577],[707,-4.438],[708,-4.5034],[709,-4.5493],[715,-4.7158],[720,-4.8123],[725,-4.8726],[732,-4.8877],[755,-4.6881],[759,-4.6822],[770,5.1431],[775,5.2329],[777,5.2363],[779,5.2054],[780,5.1787],[781,5.1345],[782,5.069],[783,4.9842],[784,4.8843],[785,4.7674],[786,4.6356],[787,4.4929],[789,4.1796],[792,3.6512],[796,2.9244],[798,2.5754],[799,2.4079],[800,2.2478],[801,2.0964],[802,1.9547],[803,1.8235],[837,1.1133],[887,3.1055],[889,2.7477],[891,2.4079],[892,2.2478],[893,2.0964],[894,1.9547],[895,1.8235],[929,1.1133]]
       }
       chest_ctrl.Scale.X {type=MovieSceneScriptingFloatChannel, default=1, const [1,751]}
       chest_ctrl.Scale.Y {type=MovieSceneScriptingFloatChannel, default=1, const [1,751]}
       chest_ctrl.Scale.Z {type=MovieSceneScriptingFloatChannel, default=1, const [1,751]}
       chest_tan_ctrl.Location.X {
        type=MovieSceneScriptingFloatChannel
        default=0
        thinned_from=751
        keys [[-2,0.0026],[-1,-0.0092],[0,-0.0108],[1,-0.0183],[2,-0.0249],[3,-0.0314],[4,-0.0375],[5,-0.0383],[6,-0.0436],[7,-0.0486],[8,-0.0545],[9,-0.0628],[10,-0.0686],[11,-0.08],[12,-0.0893],[13,-0.1],[14,-0.1101],[15,-0.1229],[16,-0.1341],[18,-0.1553],[19,-0.1563],[20,-0.1727],[22,-0.1862],[23,-0.1905],[24,-0.194],[27,-0.2001],[28,-0.209],[29,-0.2011],[30,-0.2015],[31,-0.2011],[32,-0.2017],[34,-0.1996],[35,-0.1999],[39,-0.1938],[40,-0.1919],[41,-0.1912],[43,-0.1877],[44,-0.1875],[45,-0.1858],[47,-0.1847],[48,-0.1811],[50,-0.1797],[52,-0.179],[53,-0.1824],[56,-0.1803],[57,-0.1765],[58,-0.1795],[60,-0.1774],[61,-0.1774],[62,-0.1763],[63,-0.1778],[64,-0.1773],[65,-0.1751],[66,-0.174],[67,-0.1758],[68,-0.1746],[69,-0.1766],[70,-0.1721],[71,-0.1709],[72,-0.1725],[73,-0.1716],[74,-0.1735],[75,-0.1688],[76,-0.1683],[77,-0.1701],[78,-0.1692],[79,-0.1702],[80,-0.1692],[81,-0.1653],[82,-0.1665],[83,-0.166],[84,-0.1687],[85,-0.1672],[86,-0.1677],[87,-0.1694],[88,-0.1729],[89,-0.1722],[90,-0.1744],[91,-0.1773],[92,-0.1786],[93,-0.1786],[94,-0.1774],[95,-0.1777],[96,-0.173],[97,-0.1709],[98,-0.1664],[99,-0.1612],[100,-0.1414],[101,-0.1371],[102,-0.1184],[103,-0.1108],[104,-0.0936],[105,-0.0887],[106,-0.0656],[107,-0.052],[108,-0.0455],[109,-0.0283],[110,-0.0115],[111,0.0016],[112,0.0162],[113,0.03],[114,0.0316],[115,0.0463],[116,0.058],[117,0.0567],[118,0.0723],[119,0.0722],[120,0.0844],[121,0.0833],[122,0.0834],[123,0.0918],[124,0.0917],[125,0.0946],[126,0.092],[128,0.0921],[129,0.0923],[130,0.092],[131,0.0924],[132,0.0924],[133,0.0931],[134,0.0928],[135,0.093],[136,0.0929],[138,0.0923],[139,0.0912],[140,0.0903],[141,0.0897],[143,0.0877],[148,0.084],[149,0.0825],[150,0.0818],[152,0.08],[153,0.0796],[154,0.0782],[155,0.0781],[156,0.0775],[157,0.0771],[159,0.077],[160,0.0766],[161,0.0765],[162,0.0781],[163,0.0782],[164,0.0796],[165,0.0801],[166,0.0791],[167,0.08],[168,0.0816],[171,0.0835],[172,0.0845],[173,0.0861],[174,0.0863],[175,0.0871],[176,0.0868],[177,0.0844],[178,0.0875],[179,0.0866],[180,0.0898],[181,0.0837],[182,0.0867],[183,0.0865],[184,0.0909],[185,0.0911],[186,0.0906],[187,0.0954],[188,0.0902],[189,0.0899],[190,0.0942],[191,0.0954],[192,0.0945],[194,0.0945],[195,0.0939],[197,0.0947],[198,0.0948],[199,0.0963],[200,0.0966],[201,0.0975],[202,0.0992],[203,0.0969],[204,0.0977],[205,0.0999],[206,0.0981],[207,0.1001],[208,0.0999],[209,0.0988],[210,0.0986],[211,0.0989],[212,0.0997],[213,0.0997],[214,0.1007],[215,0.1033],[216,0.1038],[217,0.1068],[218,0.1102],[220,0.1231],[221,0.1315],[222,0.1412],[223,0.1527],[224,0.1637],[225,0.1734],[226,0.184],[227,0.1953],[229,0.2093],[230,0.2147],[231,0.2163],[232,0.2171],[233,0.2151],[234,0.2104],[235,0.2076],[236,0.2009],[237,0.1971],[239,0.1847],[241,0.1696],[242,0.1613],[243,0.1537],[244,0.1451],[245,0.1355],[246,0.1283],[247,0.1188],[248,0.1102],[249,0.1026],[250,0.0937],[251,0.0839],[252,0.0753],[253,0.0684],[254,0.0602],[255,0.0518],[256,0.0423],[257,0.0338],[258,0.0266],[259,0.0166],[260,0.0079],[261,0.0012],[262,-0.0075],[263,-0.0126],[264,-0.0183],[265,-0.0217],[266,-0.0241],[267,-0.0257],[268,-0.0269],[269,-0.0305],[270,-0.0313],[271,-0.031],[272,-0.0322],[273,-0.0329],[274,-0.0325],[275,-0.0372],[276,-0.038],[277,-0.0393],[278,-0.0384],[279,-0.0396],[280,-0.0405],[281,-0.038],[282,-0.0424],[283,-0.0416],[284,-0.0411],[285,-0.0421],[286,-0.0452],[287,-0.044],[288,-0.0424],[289,-0.0467],[291,-0.0435],[292,-0.0449],[293,-0.0468],[294,-0.0455],[295,-0.0479],[296,-0.0469],[297,-0.0505],[298,-0.0493],[299,-0.0523],[300,-0.0497],[302,-0.057],[303,-0.0555],[304,-0.0566],[305,-0.061],[306,-0.0598],[307,-0.0612],[308,-0.0662],[309,-0.0647],[323,0.0026],[324,-0.0092],[325,-0.0108],[326,-0.0183],[327,-0.0249],[328,-0.0314],[329,-0.0375],[330,-0.0383],[331,-0.0436],[332,-0.0486],[333,-0.0545],[334,-0.0628],[335,-0.0686],[336,-0.08],[337,-0.0893],[338,-0.1],[339,-0.1101],[340,-0.1229],[341,-0.1341],[343,-0.1553],[344,-0.1563],[345,-0.1727],[347,-0.1862],[348,-0.1905],[349,-0.194],[352,-0.2001],[353,-0.209],[354,-0.2011],[355,-0.2015],[356,-0.2011],[357,-0.2017],[359,-0.1996],[360,-0.1999],[364,-0.1938],[365,-0.1919],[366,-0.1912],[368,-0.1877],[369,-0.1875],[370,-0.1858],[372,-0.1847],[373,-0.1811],[375,-0.1797],[377,-0.179],[378,-0.1824],[381,-0.1803],[382,-0.1765],[383,-0.1795],[385,-0.1774],[386,-0.1774],[387,-0.1763],[388,-0.1778],[389,-0.1773],[390,-0.1751],[391,-0.174],[392,-0.1758],[393,-0.1746],[394,-0.1766],[395,-0.1721],[396,-0.1709],[397,-0.1725],[398,-0.1716],[399,-0.1735],[400,-0.1688],[401,-0.1683],[402,-0.1701],[403,-0.1692],[404,-0.1702],[405,-0.1692],[406,-0.1653],[407,-0.1665],[408,-0.166],[409,-0.1687],[410,-0.1672],[411,-0.1677],[412,-0.1694],[413,-0.1729],[414,-0.1722],[415,-0.1744],[416,-0.1773],[417,-0.1786],[418,-0.1786],[419,-0.1774],[420,-0.1777],[421,-0.173],[422,-0.1709],[423,-0.1664],[424,-0.1612],[425,-0.1414],[426,-0.1371],[427,-0.1184],[428,-0.1108],[429,-0.0936],[430,-0.0887],[431,-0.0656],[432,-0.052],[433,-0.0455],[434,-0.0283],[435,-0.0115],[436,0.0016],[437,0.0162],[438,0.03],[439,0.0316],[440,0.0463],[441,0.058],[442,0.0567],[443,0.0723],[444,0.0722],[445,0.0844],[446,0.0833],[447,0.0834],[448,0.0918],[449,0.0917],[450,0.0946],[451,0.092],[453,0.0921],[454,0.0923],[455,0.092],[456,0.0924],[457,0.0924],[458,0.0931],[459,0.0928],[460,0.093],[461,0.0929],[463,0.0923],[464,0.0912],[465,0.0903],[466,0.0897],[468,0.0877],[473,0.084],[474,0.0825],[475,0.0818],[477,0.08],[478,0.0796],[479,0.0782],[480,0.0781],[481,0.0775],[482,0.0771],[484,0.077],[485,0.0766],[486,0.0765],[487,0.0781],[488,0.0782],[489,0.0796],[490,0.0801],[491,0.0791],[492,0.08],[493,0.0816],[496,0.0835],[497,0.0845],[498,0.0861],[499,0.0863],[500,0.0871],[501,0.0868],[502,0.0844],[503,0.0875],[504,0.0866],[505,0.0898],[506,0.0837],[507,0.0867],[508,0.0865],[509,0.0909],[510,0.0911],[511,0.0906],[512,0.0954],[513,0.0902],[514,0.0899],[515,0.0942],[516,0.0954],[517,0.0945],[519,0.0945],[520,0.0939],[522,0.0947],[523,0.0948],[524,0.0963],[525,0.0966],[526,0.0975],[527,0.0992],[528,0.0969],[529,0.0977],[530,0.0999],[531,0.0981],[532,0.1001],[533,0.0999],[534,0.0988],[535,0.0986],[536,0.0989],[537,0.0997],[538,0.0997],[539,0.1007],[540,0.1033],[541,0.1038],[542,0.1068],[543,0.1102],[545,0.1231],[546,0.1315],[547,0.1412],[548,0.1527],[549,0.1637],[550,0.1734],[551,0.184],[552,0.1953],[554,0.2093],[555,0.2147],[556,0.2163],[557,0.2171],[558,0.2151],[559,0.2104],[560,0.2076],[561,0.2009],[562,0.1971],[564,0.1847],[566,0.1696],[567,0.1613],[568,0.1537],[569,0.1451],[570,0.1355],[571,0.1283],[572,0.1188],[573,0.1102],[574,0.1026],[575,0.0937],[576,0.0839],[577,0.0753],[578,0.0684],[579,0.0602],[580,0.0518],[581,0.0423],[582,0.0338],[583,0.0266],[584,0.0166],[585,0.0079],[586,0.0012],[587,-0.0075],[588,-0.0126],[589,-0.0183],[590,-0.0217],[591,-0.0241],[592,-0.0257],[593,-0.0269],[594,-0.0305],[595,-0.0313],[596,-0.031],[597,-0.0322],[598,-0.0329],[599,-0.0325],[600,-0.0372],[601,-0.038],[602,-0.0393],[603,-0.0384],[604,-0.0396],[605,-0.0405],[606,-0.038],[607,-0.0424],[608,-0.0416],[609,-0.0411],[610,-0.0421],[611,-0.0452],[612,-0.044],[613,-0.0424],[614,-0.0467],[616,-0.0435],[617,-0.0449],[618,-0.0468],[619,-0.0455],[620,-0.0479],[621,-0.0469],[622,-0.0505],[623,-0.0493],[624,-0.0523],[625,-0.0497],[627,-0.057],[628,-0.0555],[629,-0.0566],[630,-0.061],[631,-0.0598],[632,-0.0612],[633,-0.0662],[676,0.0026],[677,-0.0092],[678,-0.0108],[679,-0.0183],[680,-0.0249],[681,-0.0314],[682,-0.0375],[683,-0.0383],[684,-0.0436],[685,-0.0486],[686,-0.0545],[687,-0.0628],[688,-0.0686],[689,-0.08],[690,-0.0893],[691,-0.1],[692,-0.1101],[693,-0.1229],[694,-0.1341],[696,-0.1553],[697,-0.1563],[698,-0.1727],[700,-0.1862],[701,-0.1905],[702,-0.194],[705,-0.2001],[706,-0.209],[707,-0.2011],[708,-0.2015],[709,-0.2011],[710,-0.2017],[712,-0.1996],[713,-0.1999],[717,-0.1938],[718,-0.1919],[719,-0.1912],[721,-0.1877],[722,-0.1875],[723,-0.1858],[725,-0.1847],[726,-0.1811],[728,-0.1797],[730,-0.179],[731,-0.1824],[734,-0.1803],[735,-0.1765],[736,-0.1795],[738,-0.1774],[739,-0.1774],[740,-0.1763],[741,-0.1778],[742,-0.1773],[743,-0.1751],[744,-0.174],[745,-0.1758],[746,-0.1746],[747,-0.1766],[748,-0.1721],[749,-0.1709],[750,-0.1725],[751,-0.1716],[752,-0.1735],[753,-0.1688],[754,-0.1683],[755,-0.1701],[757,-0.1702],[759,-0.1653],[770,-0.0421],[771,-0.0452],[772,-0.044],[773,-0.0424],[774,-0.0467],[776,-0.0435],[777,-0.0449],[778,-0.0468],[779,-0.0455],[780,-0.0479],[781,-0.0469],[782,-0.0505],[783,-0.0493],[784,-0.0523],[785,-0.0497],[787,-0.057],[788,-0.0555],[789,-0.0566],[790,-0.061],[791,-0.0598],[792,-0.0612],[793,-0.0662],[794,-0.0647],[795,-0.0698],[796,-0.0688],[797,-0.0739],[798,-0.0729],[799,-0.0785],[800,-0.0811],[801,-0.0861],[802,-0.0827],[803,-0.0849],[837,0.0026],[887,-0.0698],[888,-0.0688],[889,-0.0739],[890,-0.0729],[891,-0.0785],[892,-0.0811],[893,-0.0861],[894,-0.0827],[895,-0.0849],[929,0.0026]]
       }
       chest_tan_ctrl.Location.Y {
        type=MovieSceneScriptingFloatChannel
        default=-0.0303
        thinned_from=751
        keys [[-2,0.4065],[-1,0.4078],[0,0.4078],[1,0.409],[2,0.4139],[4,0.4114],[5,0.4153],[7,0.4101],[9,0.4128],[10,0.4089],[12,0.408],[13,0.4029],[14,0.4021],[15,0.3984],[16,0.3931],[20,0.3771],[21,0.37],[22,0.3695],[23,0.3614],[24,0.3568],[26,0.351],[27,0.3457],[28,0.3426],[29,0.3376],[30,0.3386],[31,0.3325],[33,0.3246],[34,0.33],[35,0.326],[38,0.316],[39,0.3198],[40,0.3172],[42,0.313],[43,0.31],[44,0.3152],[45,0.3132],[46,0.3121],[47,0.3125],[48,0.3149],[49,0.3099],[50,0.311],[51,0.314],[53,0.3176],[54,0.3202],[55,0.3188],[58,0.3285],[61,0.3405],[62,0.3391],[63,0.345],[64,0.3478],[65,0.3498],[66,0.3542],[68,0.3602],[69,0.364],[70,0.3691],[71,0.3692],[73,0.376],[75,0.3814],[76,0.3819],[82,0.393],[84,0.3988],[85,0.3997],[86,0.408],[87,0.4136],[88,0.417],[89,0.4264],[91,0.4378],[94,0.456],[95,0.4655],[96,0.4716],[97,0.4808],[98,0.4832],[99,0.4925],[100,0.5],[101,0.5091],[102,0.5133],[103,0.5219],[105,0.5352],[107,0.5453],[110,0.5561],[111,0.5548],[112,0.5568],[114,0.5572],[115,0.5547],[116,0.5541],[118,0.5408],[120,0.5304],[121,0.5215],[122,0.5148],[124,0.4951],[125,0.484],[126,0.4784],[127,0.4667],[128,0.458],[130,0.4447],[131,0.4399],[132,0.4365],[133,0.4268],[134,0.4223],[135,0.4188],[136,0.4191],[137,0.4247],[138,0.4243],[140,0.4297],[141,0.4342],[142,0.4319],[143,0.4416],[145,0.4514],[146,0.4543],[148,0.4646],[149,0.4674],[150,0.4726],[152,0.4802],[154,0.486],[155,0.4875],[156,0.4939],[157,0.4959],[158,0.4995],[159,0.5009],[160,0.5058],[162,0.504],[164,0.5091],[166,0.5113],[167,0.5068],[168,0.5119],[169,0.5104],[171,0.5144],[172,0.5112],[174,0.5118],[175,0.5082],[176,0.5087],[178,0.5076],[179,0.5011],[181,0.4992],[182,0.4968],[183,0.4965],[185,0.4839],[186,0.4885],[188,0.4769],[189,0.4776],[190,0.4725],[191,0.4699],[192,0.4688],[193,0.4658],[194,0.466],[195,0.4636],[196,0.4546],[197,0.4528],[198,0.4547],[199,0.4534],[200,0.4557],[201,0.4455],[202,0.4451],[203,0.448],[204,0.4486],[205,0.4422],[207,0.4429],[208,0.4395],[209,0.4417],[210,0.4414],[211,0.4448],[213,0.4423],[215,0.4468],[216,0.4443],[217,0.4494],[218,0.4477],[219,0.4495],[221,0.4454],[222,0.4512],[223,0.4499],[224,0.4463],[225,0.456],[226,0.4514],[227,0.4583],[228,0.4528],[229,0.4565],[230,0.4615],[231,0.46],[232,0.46],[233,0.4706],[234,0.4705],[235,0.4785],[236,0.4789],[237,0.4905],[238,0.49],[239,0.5002],[241,0.5105],[242,0.5239],[244,0.5346],[245,0.5481],[246,0.5543],[247,0.5664],[248,0.5711],[250,0.5917],[251,0.5936],[253,0.6091],[254,0.6156],[255,0.6239],[258,0.6408],[260,0.6487],[262,0.6552],[263,0.6564],[264,0.6593],[265,0.6594],[266,0.661],[269,0.662],[271,0.6611],[272,0.6571],[273,0.662],[274,0.657],[275,0.6554],[277,0.6462],[278,0.6469],[280,0.6455],[281,0.6411],[282,0.6408],[283,0.6465],[284,0.6425],[286,0.6428],[288,0.6417],[292,0.6447],[294,0.6499],[295,0.6507],[296,0.6542],[298,0.6588],[299,0.6627],[302,0.6707],[303,0.6719],[304,0.6768],[305,0.6786],[306,0.6837],[307,0.6852],[308,0.6889],[309,0.69],[323,0.4065],[324,0.4078],[325,0.4078],[326,0.409],[327,0.4139],[329,0.4114],[330,0.4153],[332,0.4101],[334,0.4128],[335,0.4089],[337,0.408],[338,0.4029],[339,0.4021],[340,0.3984],[341,0.3931],[345,0.3771],[346,0.37],[347,0.3695],[348,0.3614],[349,0.3568],[351,0.351],[352,0.3457],[353,0.3426],[354,0.3376],[355,0.3386],[356,0.3325],[358,0.3246],[359,0.33],[360,0.326],[363,0.316],[364,0.3198],[365,0.3172],[367,0.313],[368,0.31],[369,0.3152],[370,0.3132],[371,0.3121],[372,0.3125],[373,0.3149],[374,0.3099],[375,0.311],[376,0.314],[378,0.3176],[379,0.3202],[380,0.3188],[383,0.3285],[386,0.3405],[387,0.3391],[388,0.345],[389,0.3478],[390,0.3498],[391,0.3542],[393,0.3602],[394,0.364],[395,0.3691],[396,0.3692],[398,0.376],[400,0.3814],[401,0.3819],[407,0.393],[409,0.3988],[410,0.3997],[411,0.408],[412,0.4136],[413,0.417],[414,0.4264],[416,0.4378],[419,0.456],[420,0.4655],[421,0.4716],[422,0.4808],[423,0.4832],[424,0.4925],[425,0.5],[426,0.5091],[427,0.5133],[428,0.5219],[430,0.5352],[432,0.5453],[435,0.5561],[436,0.5548],[437,0.5568],[439,0.5572],[440,0.5547],[441,0.5541],[443,0.5408],[445,0.5304],[446,0.5215],[447,0.5148],[449,0.4951],[450,0.484],[451,0.4784],[452,0.4667],[453,0.458],[455,0.4447],[456,0.4399],[457,0.4365],[458,0.4268],[459,0.4223],[460,0.4188],[461,0.4191],[462,0.4247],[463,0.4243],[465,0.4297],[466,0.4342],[467,0.4319],[468,0.4416],[470,0.4514],[471,0.4543],[473,0.4646],[474,0.4674],[475,0.4726],[477,0.4802],[479,0.486],[480,0.4875],[481,0.4939],[482,0.4959],[483,0.4995],[484,0.5009],[485,0.5058],[487,0.504],[489,0.5091],[491,0.5113],[492,0.5068],[493,0.5119],[494,0.5104],[496,0.5144],[497,0.5112],[499,0.5118],[500,0.5082],[501,0.5087],[503,0.5076],[504,0.5011],[506,0.4992],[507,0.4968],[508,0.4965],[510,0.4839],[511,0.4885],[513,0.4769],[514,0.4776],[515,0.4725],[516,0.4699],[517,0.4688],[518,0.4658],[519,0.466],[520,0.4636],[521,0.4546],[522,0.4528],[523,0.4547],[524,0.4534],[525,0.4557],[526,0.4455],[527,0.4451],[528,0.448],[529,0.4486],[530,0.4422],[532,0.4429],[533,0.4395],[534,0.4417],[535,0.4414],[536,0.4448],[538,0.4423],[540,0.4468],[541,0.4443],[542,0.4494],[543,0.4477],[544,0.4495],[546,0.4454],[547,0.4512],[548,0.4499],[549,0.4463],[550,0.456],[551,0.4514],[552,0.4583],[553,0.4528],[554,0.4565],[555,0.4615],[556,0.46],[557,0.46],[558,0.4706],[559,0.4705],[560,0.4785],[561,0.4789],[562,0.4905],[563,0.49],[564,0.5002],[566,0.5105],[567,0.5239],[569,0.5346],[570,0.5481],[571,0.5543],[572,0.5664],[573,0.5711],[575,0.5917],[576,0.5936],[578,0.6091],[579,0.6156],[580,0.6239],[583,0.6408],[585,0.6487],[587,0.6552],[588,0.6564],[589,0.6593],[590,0.6594],[591,0.661],[594,0.662],[596,0.6611],[597,0.6571],[598,0.662],[599,0.657],[600,0.6554],[602,0.6462],[603,0.6469],[605,0.6455],[606,0.6411],[607,0.6408],[608,0.6465],[609,0.6425],[611,0.6428],[613,0.6417],[617,0.6447],[619,0.6499],[620,0.6507],[621,0.6542],[623,0.6588],[624,0.6627],[627,0.6707],[628,0.6719],[629,0.6768],[630,0.6786],[631,0.6837],[632,0.6852],[633,0.6889],[634,0.69],[676,0.4065],[677,0.4078],[678,0.4078],[679,0.409],[680,0.4139],[682,0.4114],[683,0.4153],[685,0.4101],[687,0.4128],[688,0.4089],[690,0.408],[691,0.4029],[692,0.4021],[693,0.3984],[694,0.3931],[698,0.3771],[699,0.37],[700,0.3695],[701,0.3614],[702,0.3568],[704,0.351],[705,0.3457],[706,0.3426],[707,0.3376],[708,0.3386],[709,0.3325],[711,0.3246],[712,0.33],[713,0.326],[716,0.316],[717,0.3198],[718,0.3172],[720,0.313],[721,0.31],[722,0.3152],[723,0.3132],[724,0.3121],[725,0.3125],[726,0.3149],[727,0.3099],[728,0.311],[729,0.314],[731,0.3176],[732,0.3202],[733,0.3188],[736,0.3285],[739,0.3405],[740,0.3391],[741,0.345],[742,0.3478],[743,0.3498],[744,0.3542],[746,0.3602],[747,0.364],[748,0.3691],[749,0.3692],[751,0.376],[753,0.3814],[754,0.3819],[759,0.391],[770,0.6426],[774,0.6422],[777,0.6447],[779,0.6499],[780,0.6507],[781,0.6542],[783,0.6588],[784,0.6627],[787,0.6707],[788,0.6719],[789,0.6768],[790,0.6786],[791,0.6837],[792,0.6852],[793,0.6889],[794,0.69],[795,0.6945],[796,0.6953],[797,0.6994],[799,0.7002],[800,0.7044],[802,0.7051],[803,0.7082],[837,0.4065],[887,0.6945],[888,0.6953],[889,0.6994],[891,0.7002],[892,0.7044],[894,0.7051],[895,0.7082],[929,0.4065]]
       }
       chest_tan_ctrl.Location.Z {
        type=MovieSceneScriptingFloatChannel
        default=0.265
        thinned_from=751
        keys [[-2,0.246],[-1,0.2466],[1,0.2465],[2,0.2487],[5,0.2491],[6,0.2473],[8,0.2478],[9,0.2475],[10,0.2462],[12,0.2443],[14,0.2437],[15,0.2427],[16,0.2439],[17,0.2406],[18,0.2345],[19,0.2386],[20,0.2365],[21,0.2383],[22,0.2351],[23,0.2366],[24,0.2327],[25,0.2334],[26,0.2258],[27,0.2295],[28,0.2307],[29,0.2261],[30,0.2254],[32,0.226],[34,0.2231],[35,0.2241],[36,0.223],[38,0.2222],[39,0.221],[40,0.2222],[45,0.2212],[46,0.2215],[47,0.2205],[50,0.2203],[51,0.2269],[52,0.2238],[56,0.2238],[57,0.2273],[60,0.2284],[61,0.2315],[63,0.232],[64,0.2354],[65,0.2323],[67,0.2391],[68,0.2391],[69,0.2396],[71,0.2395],[72,0.2429],[73,0.2429],[74,0.2463],[75,0.2428],[76,0.2431],[77,0.2464],[80,0.2464],[81,0.2484],[82,0.2474],[83,0.2492],[84,0.2486],[86,0.2518],[87,0.2546],[88,0.2566],[89,0.2597],[90,0.2621],[91,0.2673],[93,0.2715],[94,0.2746],[95,0.2793],[96,0.2798],[97,0.2849],[98,0.2873],[99,0.2929],[100,0.2937],[101,0.2961],[102,0.3006],[103,0.3029],[104,0.307],[106,0.3116],[107,0.3162],[108,0.3144],[109,0.3154],[110,0.3176],[111,0.319],[112,0.3181],[113,0.3186],[114,0.3178],[115,0.3208],[116,0.3159],[117,0.3125],[118,0.3113],[119,0.3055],[120,0.3022],[121,0.3002],[122,0.2937],[124,0.2852],[125,0.2794],[126,0.2766],[128,0.2657],[131,0.255],[132,0.2519],[133,0.2502],[135,0.2442],[136,0.2458],[138,0.2455],[139,0.2447],[140,0.2457],[141,0.2476],[142,0.247],[143,0.2499],[144,0.2518],[147,0.2553],[148,0.2573],[149,0.2587],[150,0.2606],[151,0.2619],[152,0.2638],[153,0.2649],[154,0.2666],[155,0.2662],[157,0.27],[159,0.2721],[160,0.2721],[161,0.2733],[163,0.2737],[164,0.2749],[165,0.2749],[166,0.2737],[168,0.2739],[169,0.2731],[170,0.2734],[172,0.2722],[173,0.2707],[174,0.2706],[176,0.2682],[177,0.2655],[178,0.2641],[179,0.2666],[180,0.2638],[181,0.2628],[182,0.2594],[183,0.2584],[184,0.2592],[185,0.2581],[186,0.2555],[187,0.2545],[188,0.2522],[189,0.2551],[190,0.2543],[191,0.2523],[192,0.2525],[193,0.2509],[194,0.2505],[195,0.2513],[196,0.2506],[197,0.2514],[198,0.249],[199,0.2492],[200,0.25],[201,0.2498],[203,0.2503],[204,0.2513],[205,0.2509],[206,0.2491],[207,0.2504],[212,0.2523],[215,0.2545],[216,0.2546],[218,0.2559],[220,0.2559],[222,0.2594],[223,0.2621],[224,0.264],[225,0.2645],[226,0.2672],[227,0.2732],[228,0.2724],[229,0.2766],[230,0.2792],[231,0.2795],[232,0.2848],[233,0.285],[234,0.2881],[236,0.2924],[237,0.2972],[239,0.3012],[240,0.3096],[241,0.3084],[242,0.3177],[243,0.3167],[244,0.3255],[245,0.3253],[246,0.3261],[247,0.3343],[248,0.3335],[249,0.343],[250,0.3427],[251,0.3513],[252,0.3506],[253,0.3595],[254,0.3586],[256,0.3657],[258,0.3696],[259,0.3726],[260,0.374],[261,0.3767],[262,0.3774],[263,0.3809],[264,0.381],[266,0.3836],[269,0.3852],[270,0.3852],[271,0.3861],[272,0.385],[275,0.3856],[276,0.3841],[277,0.381],[278,0.3824],[279,0.3825],[281,0.3812],[282,0.3816],[284,0.3804],[285,0.3823],[286,0.3826],[287,0.3815],[290,0.3833],[292,0.3853],[293,0.3855],[294,0.3869],[295,0.3875],[298,0.3912],[300,0.3948],[301,0.3952],[302,0.397],[303,0.397],[305,0.4022],[306,0.4027],[307,0.4045],[308,0.4049],[309,0.4071],[323,0.246],[324,0.2466],[326,0.2465],[327,0.2487],[330,0.2491],[331,0.2473],[333,0.2478],[334,0.2475],[335,0.2462],[337,0.2443],[339,0.2437],[340,0.2427],[341,0.2439],[342,0.2406],[343,0.2345],[344,0.2386],[345,0.2365],[346,0.2383],[347,0.2351],[348,0.2366],[349,0.2327],[350,0.2334],[351,0.2258],[352,0.2295],[353,0.2307],[354,0.2261],[355,0.2254],[357,0.226],[359,0.2231],[360,0.2241],[361,0.223],[363,0.2222],[364,0.221],[365,0.2222],[370,0.2212],[371,0.2215],[372,0.2205],[375,0.2203],[376,0.2269],[377,0.2238],[381,0.2238],[382,0.2273],[385,0.2284],[386,0.2315],[388,0.232],[389,0.2354],[390,0.2323],[392,0.2391],[393,0.2391],[394,0.2396],[396,0.2395],[397,0.2429],[398,0.2429],[399,0.2463],[400,0.2428],[401,0.2431],[402,0.2464],[405,0.2464],[406,0.2484],[407,0.2474],[408,0.2492],[409,0.2486],[411,0.2518],[412,0.2546],[413,0.2566],[414,0.2597],[415,0.2621],[416,0.2673],[418,0.2715],[419,0.2746],[420,0.2793],[421,0.2798],[422,0.2849],[423,0.2873],[424,0.2929],[425,0.2937],[426,0.2961],[427,0.3006],[428,0.3029],[429,0.307],[431,0.3116],[432,0.3162],[433,0.3144],[434,0.3154],[435,0.3176],[436,0.319],[437,0.3181],[438,0.3186],[439,0.3178],[440,0.3208],[441,0.3159],[442,0.3125],[443,0.3113],[444,0.3055],[445,0.3022],[446,0.3002],[447,0.2937],[449,0.2852],[450,0.2794],[451,0.2766],[453,0.2657],[456,0.255],[457,0.2519],[458,0.2502],[460,0.2442],[461,0.2458],[463,0.2455],[464,0.2447],[465,0.2457],[466,0.2476],[467,0.247],[468,0.2499],[469,0.2518],[472,0.2553],[473,0.2573],[474,0.2587],[475,0.2606],[476,0.2619],[477,0.2638],[478,0.2649],[479,0.2666],[480,0.2662],[482,0.27],[484,0.2721],[485,0.2721],[486,0.2733],[488,0.2737],[489,0.2749],[490,0.2749],[491,0.2737],[493,0.2739],[494,0.2731],[495,0.2734],[497,0.2722],[498,0.2707],[499,0.2706],[501,0.2682],[502,0.2655],[503,0.2641],[504,0.2666],[505,0.2638],[506,0.2628],[507,0.2594],[508,0.2584],[509,0.2592],[510,0.2581],[511,0.2555],[512,0.2545],[513,0.2522],[514,0.2551],[515,0.2543],[516,0.2523],[517,0.2525],[518,0.2509],[519,0.2505],[520,0.2513],[521,0.2506],[522,0.2514],[523,0.249],[524,0.2492],[525,0.25],[526,0.2498],[528,0.2503],[529,0.2513],[530,0.2509],[531,0.2491],[532,0.2504],[537,0.2523],[540,0.2545],[541,0.2546],[543,0.2559],[545,0.2559],[547,0.2594],[548,0.2621],[549,0.264],[550,0.2645],[551,0.2672],[552,0.2732],[553,0.2724],[554,0.2766],[555,0.2792],[556,0.2795],[557,0.2848],[558,0.285],[559,0.2881],[561,0.2924],[562,0.2972],[564,0.3012],[565,0.3096],[566,0.3084],[567,0.3177],[568,0.3167],[569,0.3255],[570,0.3253],[571,0.3261],[572,0.3343],[573,0.3335],[574,0.343],[575,0.3427],[576,0.3513],[577,0.3506],[578,0.3595],[579,0.3586],[581,0.3657],[583,0.3696],[584,0.3726],[585,0.374],[586,0.3767],[587,0.3774],[588,0.3809],[589,0.381],[591,0.3836],[594,0.3852],[595,0.3852],[596,0.3861],[597,0.385],[600,0.3856],[601,0.3841],[602,0.381],[603,0.3824],[604,0.3825],[606,0.3812],[607,0.3816],[609,0.3804],[610,0.3823],[611,0.3826],[612,0.3815],[615,0.3833],[617,0.3853],[618,0.3855],[619,0.3869],[620,0.3875],[623,0.3912],[625,0.3948],[626,0.3952],[627,0.397],[628,0.397],[630,0.4022],[631,0.4027],[632,0.4045],[633,0.4049],[634,0.4071],[676,0.246],[677,0.2466],[679,0.2465],[680,0.2487],[683,0.2491],[684,0.2473],[686,0.2478],[687,0.2475],[688,0.2462],[690,0.2443],[692,0.2437],[693,0.2427],[694,0.2439],[695,0.2406],[696,0.2345],[697,0.2386],[698,0.2365],[699,0.2383],[700,0.2351],[701,0.2366],[702,0.2327],[703,0.2334],[704,0.2258],[705,0.2295],[706,0.2307],[707,0.2261],[708,0.2254],[710,0.226],[712,0.2231],[713,0.2241],[714,0.223],[716,0.2222],[717,0.221],[718,0.2222],[723,0.2212],[724,0.2215],[725,0.2205],[728,0.2203],[729,0.2269],[730,0.2238],[734,0.2238],[735,0.2273],[738,0.2284],[739,0.2315],[741,0.232],[742,0.2354],[743,0.2323],[745,0.2391],[746,0.2391],[747,0.2396],[749,0.2395],[750,0.2429],[751,0.2429],[752,0.2463],[753,0.2428],[754,0.2431],[755,0.2464],[757,0.2464],[759,0.2484],[770,0.3823],[771,0.3826],[772,0.3815],[775,0.3833],[777,0.3853],[778,0.3855],[779,0.3869],[780,0.3875],[783,0.3912],[785,0.3948],[786,0.3952],[787,0.397],[788,0.397],[790,0.4022],[791,0.4027],[792,0.4045],[793,0.4049],[794,0.4071],[795,0.4075],[796,0.4098],[798,0.4099],[799,0.4128],[801,0.4129],[802,0.4154],[803,0.4154],[837,0.246],[887,0.4075],[888,0.4098],[890,0.4099],[891,0.4128],[893,0.4129],[894,0.4154],[895,0.4154],[929,0.246]]
       }
       chest_tan_ctrl.Rotation.X {
        type=MovieSceneScriptingFloatChannel
        default=0
        thinned_from=751
        keys [[-2,1.4612],[-1,1.5405],[0,1.5575],[1,1.5668],[2,1.5712],[3,1.5693],[4,1.5614],[5,1.5482],[6,1.5302],[7,1.5057],[8,1.4697],[9,1.4215],[10,1.3599],[11,1.2804],[12,1.1838],[13,1.0681],[14,0.936],[15,0.7893],[16,0.6279],[17,0.4603],[18,0.2848],[19,0.1081],[20,0],[64,0],[65,0.0103],[66,0.09],[67,0.1664],[68,0.2409],[70,0.3749],[71,0.438],[72,0.494],[73,0.5459],[74,0.5928],[75,0.6359],[76,0.6754],[77,0.7114],[79,0.7785],[81,0.8439],[82,0.8809],[83,0.9243],[84,0.973],[85,1.039],[86,1.1384],[87,1.2667],[88,1.4178],[89,1.5882],[90,1.7757],[91,1.9777],[92,2.1917],[95,2.8674],[96,3.0983],[99,3.8221],[100,4.0578],[101,4.2841],[102,4.494],[103,4.6816],[104,4.8414],[105,4.9688],[106,5.0599],[107,5.1112],[108,5.1316],[109,5.1329],[110,5.1169],[111,5.0826],[112,5.0302],[113,4.9606],[114,4.8564],[115,4.7065],[116,4.5167],[117,4.2915],[118,4.0409],[119,3.7667],[120,3.4784],[123,2.5802],[124,2.2882],[125,2.0083],[126,1.743],[127,1.4981],[128,1.2758],[129,1.0754],[130,0.8998],[131,0.7452],[132,0.6164],[133,0.5099],[134,0.4259],[135,0.3672],[136,0.3322],[137,0.3217],[138,0.3301],[139,0.3489],[140,0.3779],[141,0.4141],[142,0.459],[143,0.5116],[144,0.5723],[145,0.6369],[147,0.7813],[150,1.0084],[152,1.1511],[153,1.2199],[155,1.3375],[156,1.3908],[157,1.4398],[159,1.5248],[160,1.5649],[161,1.5956],[162,1.6168],[164,1.6374],[165,1.6343],[167,1.6144],[168,1.5906],[169,1.5621],[170,1.526],[172,1.4375],[174,1.3319],[178,1.1072],[180,1.0012],[184,0.7973],[186,0.705],[187,0.6631],[188,0.6258],[189,0.5928],[190,0.5641],[191,0.5394],[192,0.5218],[193,0.5065],[194,0.4954],[195,0.4856],[196,0.4791],[197,0.4747],[198,0.4717],[199,0.4706],[200,0.472],[201,0.4725],[202,0.476],[204,0.4859],[206,0.4985],[207,0.5074],[208,0.5194],[209,0.5341],[210,0.5523],[211,0.573],[213,0.6222],[214,0.6424],[215,0.6542],[216,0.659],[217,0.654],[218,0.6359],[219,0.6038],[220,0.5597],[221,0.5267],[222,0.5006],[223,0.4834],[224,0.4759],[225,0.483],[226,0.5072],[227,0.5576],[228,0.6325],[229,0.7324],[230,0.8586],[231,1.0121],[232,1.193],[233,1.4005],[234,1.6263],[235,1.8681],[236,2.1227],[237,2.3912],[238,2.6703],[239,2.9601],[240,3.2622],[246,5.1593],[247,5.4683],[248,5.7618],[249,6.0433],[250,6.3096],[251,6.562],[252,6.7988],[254,7.23],[256,7.6046],[258,7.9422],[260,8.2435],[262,8.5078],[264,8.7305],[266,8.9107],[268,9.0488],[270,9.1475],[272,9.2092],[281,9.2403],[289,9.392],[297,9.7364],[309,10.3922],[323,1.4612],[324,1.5405],[325,1.5575],[326,1.5668],[327,1.5712],[328,1.5693],[329,1.5614],[330,1.5482],[331,1.5302],[332,1.5057],[333,1.4697],[334,1.4215],[335,1.3599],[336,1.2804],[337,1.1838],[338,1.0681],[339,0.936],[340,0.7893],[341,0.6279],[342,0.4603],[343,0.2848],[344,0.1081],[345,0],[389,0],[390,0.0103],[391,0.09],[392,0.1664],[393,0.2409],[395,0.3749],[396,0.438],[397,0.494],[398,0.5459],[399,0.5928],[400,0.6359],[401,0.6754],[402,0.7114],[404,0.7785],[406,0.8439],[407,0.8809],[408,0.9243],[409,0.973],[410,1.039],[411,1.1384],[412,1.2667],[413,1.4178],[414,1.5882],[415,1.7757],[416,1.9777],[417,2.1917],[420,2.8674],[421,3.0983],[424,3.8221],[425,4.0578],[426,4.2841],[427,4.494],[428,4.6816],[429,4.8414],[430,4.9688],[431,5.0599],[432,5.1112],[433,5.1316],[434,5.1329],[435,5.1169],[436,5.0826],[437,5.0302],[438,4.9606],[439,4.8564],[440,4.7065],[441,4.5167],[442,4.2915],[443,4.0409],[444,3.7667],[445,3.4784],[448,2.5802],[449,2.2882],[450,2.0083],[451,1.743],[452,1.4981],[453,1.2758],[454,1.0754],[455,0.8998],[456,0.7452],[457,0.6164],[458,0.5099],[459,0.4259],[460,0.3672],[461,0.3322],[462,0.3217],[463,0.3301],[464,0.3489],[465,0.3779],[466,0.4141],[467,0.459],[468,0.5116],[469,0.5723],[470,0.6369],[472,0.7813],[475,1.0084],[477,1.1511],[478,1.2199],[480,1.3375],[481,1.3908],[482,1.4398],[484,1.5248],[485,1.5649],[486,1.5956],[487,1.6168],[489,1.6374],[490,1.6343],[492,1.6144],[493,1.5906],[494,1.5621],[495,1.526],[497,1.4375],[499,1.3319],[503,1.1072],[505,1.0012],[509,0.7973],[511,0.705],[512,0.6631],[513,0.6258],[514,0.5928],[515,0.5641],[516,0.5394],[517,0.5218],[518,0.5065],[519,0.4954],[520,0.4856],[521,0.4791],[522,0.4747],[523,0.4717],[524,0.4706],[525,0.472],[526,0.4725],[527,0.476],[529,0.4859],[531,0.4985],[532,0.5074],[533,0.5194],[534,0.5341],[535,0.5523],[536,0.573],[538,0.6222],[539,0.6424],[540,0.6542],[541,0.659],[542,0.654],[543,0.6359],[544,0.6038],[545,0.5597],[546,0.5267],[547,0.5006],[548,0.4834],[549,0.4759],[550,0.483],[551,0.5072],[552,0.5576],[553,0.6325],[554,0.7324],[555,0.8586],[556,1.0121],[557,1.193],[558,1.4005],[559,1.6263],[560,1.8681],[561,2.1227],[562,2.3912],[563,2.6703],[564,2.9601],[565,3.2622],[571,5.1593],[572,5.4683],[573,5.7618],[574,6.0433],[575,6.3096],[576,6.562],[577,6.7988],[579,7.23],[581,7.6046],[583,7.9422],[585,8.2435],[587,8.5078],[589,8.7305],[591,8.9107],[593,9.0488],[595,9.1475],[597,9.2092],[606,9.2403],[614,9.392],[622,9.7364],[634,10.3922],[676,1.4612],[677,1.5405],[678,1.5575],[679,1.5668],[680,1.5712],[681,1.5693],[682,1.5614],[683,1.5482],[684,1.5302],[685,1.5057],[686,1.4697],[687,1.4215],[688,1.3599],[689,1.2804],[690,1.1838],[691,1.0681],[692,0.936],[693,0.7893],[694,0.6279],[695,0.4603],[696,0.2848],[697,0.1081],[698,0],[742,0],[743,0.0103],[744,0.09],[745,0.1664],[746,0.2409],[748,0.3749],[749,0.438],[750,0.494],[751,0.5459],[752,0.5928],[753,0.6359],[754,0.6754],[755,0.7114],[757,0.7785],[759,0.8439],[770,9.3],[775,9.4251],[796,10.4797],[803,10.6976],[837,1.4612],[887,10.4372],[895,10.6976],[929,1.4612]]
       }
       chest_tan_ctrl.Rotation.Y {
        type=MovieSceneScriptingFloatChannel
        default=0
        thinned_from=751
        keys [[-2,2.5399],[-1,2.8933],[0,3.0666],[1,3.2303],[2,3.3836],[3,3.5288],[4,3.665],[6,3.9127],[7,4.0254],[8,4.1567],[9,4.3403],[10,4.5818],[11,4.8697],[12,5.1995],[13,5.5637],[14,5.9553],[15,6.3612],[18,7.5999],[19,7.9922],[20,8.3621],[21,8.7009],[22,8.9977],[23,9.2461],[24,9.4372],[26,9.7153],[27,9.823],[29,9.9751],[31,10.0512],[34,10.069],[53,9.5367],[81,8.2657],[83,8.206],[84,8.201],[85,8.2187],[87,8.304],[91,8.5968],[92,8.6522],[93,8.6749],[94,8.6551],[95,8.5806],[96,8.4352],[97,8.2085],[98,7.9121],[99,7.5589],[100,7.148],[101,6.6874],[102,6.1812],[103,5.631],[104,5.0452],[108,2.5856],[109,1.9907],[110,1.4171],[111,0.8691],[112,0.3525],[113,0],[262,0],[263,0.1313],[264,0.2755],[265,0.3861],[266,0.47],[267,0.5179],[269,0.582],[270,0.612],[271,0.6447],[273,0.7221],[274,0.7586],[275,0.8004],[276,0.8362],[277,0.8739],[279,0.9441],[280,0.9845],[281,1.0218],[282,1.0629],[283,1.1011],[285,1.183],[286,1.2219],[287,1.2666],[288,1.3054],[290,1.3888],[292,1.4827],[295,1.6422],[296,1.7027],[297,1.7733],[298,1.8521],[299,1.9355],[301,2.1177],[309,2.9248],[323,2.5399],[324,2.8933],[325,3.0666],[326,3.2303],[327,3.3836],[328,3.5288],[329,3.665],[331,3.9127],[332,4.0254],[333,4.1567],[334,4.3403],[335,4.5818],[336,4.8697],[337,5.1995],[338,5.5637],[339,5.9553],[340,6.3612],[343,7.5999],[344,7.9922],[345,8.3621],[346,8.7009],[347,8.9977],[348,9.2461],[349,9.4372],[351,9.7153],[352,9.823],[354,9.9751],[356,10.0512],[359,10.069],[378,9.5367],[406,8.2657],[408,8.206],[409,8.201],[410,8.2187],[412,8.304],[416,8.5968],[417,8.6522],[418,8.6749],[419,8.6551],[420,8.5806],[421,8.4352],[422,8.2085],[423,7.9121],[424,7.5589],[425,7.148],[426,6.6874],[427,6.1812],[428,5.631],[429,5.0452],[433,2.5856],[434,1.9907],[435,1.4171],[436,0.8691],[437,0.3525],[438,0],[587,0],[588,0.1313],[589,0.2755],[590,0.3861],[591,0.47],[592,0.5179],[594,0.582],[595,0.612],[596,0.6447],[598,0.7221],[599,0.7586],[600,0.8004],[601,0.8362],[602,0.8739],[604,0.9441],[605,0.9845],[606,1.0218],[607,1.0629],[608,1.1011],[610,1.183],[611,1.2219],[612,1.2666],[613,1.3054],[615,1.3888],[617,1.4827],[620,1.6422],[621,1.7027],[622,1.7733],[623,1.8521],[624,1.9355],[626,2.1177],[634,2.9248],[676,2.5399],[677,2.8933],[678,3.0666],[679,3.2303],[680,3.3836],[681,3.5288],[682,3.665],[684,3.9127],[685,4.0254],[686,4.1567],[687,4.3403],[688,4.5818],[689,4.8697],[690,5.1995],[691,5.5637],[692,5.9553],[693,6.3612],[696,7.5999],[697,7.9922],[698,8.3621],[699,8.7009],[700,8.9977],[701,9.2461],[702,9.4372],[704,9.7153],[705,9.823],[707,9.9751],[709,10.0512],[712,10.069],[731,9.5367],[755,8.424],[759,8.2657],[770,1.183],[771,1.2219],[772,1.2666],[773,1.3054],[775,1.3888],[777,1.4827],[780,1.6422],[781,1.7027],[782,1.7733],[783,1.8521],[784,1.9355],[786,2.1177],[795,3.0256],[797,3.2165],[799,3.3931],[801,3.5484],[803,3.6877],[837,2.5399],[887,3.0256],[889,3.2165],[891,3.3931],[893,3.5484],[895,3.6877],[929,2.5399]]
       }
       chest_tan_ctrl.Rotation.Z {
        type=MovieSceneScriptingFloatChannel
        default=0
        thinned_from=751
        keys [[-2,0.7914],[-1,0.7958],[0,0.7533],[1,0.6951],[2,0.6275],[3,0.5479],[4,0.4602],[5,0.3615],[6,0.2531],[7,0.1386],[8,0.0153],[9,0],[224,0],[225,0.0506],[226,0.5542],[227,1.0531],[228,1.5395],[229,1.9975],[230,2.4226],[231,2.8058],[232,3.1404],[233,3.4152],[234,3.6447],[235,3.8416],[236,4.0116],[237,4.1514],[238,4.2638],[239,4.3511],[240,4.4136],[241,4.4531],[242,4.472],[243,4.4722],[244,4.4551],[245,4.4209],[246,4.3726],[247,4.3115],[248,4.2381],[249,4.1561],[251,3.9635],[256,3.4165],[257,3.3114],[259,3.1224],[260,3.0435],[261,2.9709],[262,2.9147],[263,2.8667],[264,2.836],[265,2.8216],[266,2.821],[267,2.8424],[270,2.9221],[274,3.0027],[290,3.2006],[292,3.2002],[294,3.1785],[295,3.1602],[296,3.1322],[297,3.0882],[298,3.0353],[299,2.9701],[300,2.8955],[301,2.8115],[302,2.7218],[304,2.5219],[307,2.1876],[309,1.9584],[323,0.7914],[324,0.7958],[325,0.7533],[326,0.6951],[327,0.6275],[328,0.5479],[329,0.4602],[330,0.3615],[331,0.2531],[332,0.1386],[333,0.0153],[334,0],[549,0],[550,0.0506],[551,0.5542],[552,1.0531],[553,1.5395],[554,1.9975],[555,2.4226],[556,2.8058],[557,3.1404],[558,3.4152],[559,3.6447],[560,3.8416],[561,4.0116],[562,4.1514],[563,4.2638],[564,4.3511],[565,4.4136],[566,4.4531],[567,4.472],[568,4.4722],[569,4.4551],[570,4.4209],[571,4.3726],[572,4.3115],[573,4.2381],[574,4.1561],[576,3.9635],[581,3.4165],[582,3.3114],[584,3.1224],[585,3.0435],[586,2.9709],[587,2.9147],[588,2.8667],[589,2.836],[590,2.8216],[591,2.821],[592,2.8424],[595,2.9221],[599,3.0027],[615,3.2006],[617,3.2002],[619,3.1785],[620,3.1602],[621,3.1322],[622,3.0882],[623,3.0353],[624,2.9701],[625,2.8955],[626,2.8115],[627,2.7218],[629,2.5219],[632,2.1876],[634,1.9584],[676,0.7914],[677,0.7958],[678,0.7533],[679,0.6951],[680,0.6275],[681,0.5479],[682,0.4602],[683,0.3615],[684,0.2531],[685,0.1386],[686,0.0153],[687,0],[759,0],[770,3.1479],[775,3.2006],[777,3.2002],[779,3.1785],[780,3.1602],[781,3.1322],[782,3.0882],[783,3.0353],[784,2.9701],[785,2.8955],[786,2.8115],[787,2.7218],[789,2.5219],[792,2.1876],[797,1.6164],[798,1.5079],[799,1.4025],[800,1.3008],[801,1.206],[802,1.1168],[803,1.0344],[837,0.7914],[887,1.8431],[889,1.6164],[890,1.5079],[891,1.4025],[892,1.3008],[893,1.206],[894,1.1168],[895,1.0344],[929,0.7914]]
       }
       chest_tan_ctrl.Scale.X {type=MovieSceneScriptingFloatChannel, default=1, const [1,751]}
       chest_tan_ctrl.Scale.Y {type=MovieSceneScriptingFloatChannel, default=1, const [1,751]}
       chest_tan_ctrl.Scale.Z {type=MovieSceneScriptingFloatChannel, default=1, const [1,751]}
       Weight {type=MovieSceneScriptingFloatChannel, default=1}
      ]
     }]
   }
  ]
 }]
display_rate {denominator=1, numerator=30}
tick_resolution {denominator=1, numerator=24000}
tracks [{type=MovieSceneCameraCutTrack, sections [{type=MovieSceneCameraCutSection, end_frame=320, start_frame=0}]}]
